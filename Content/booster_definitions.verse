
using { /Verse.org/Simulation }
using { /Verse.org/Simulation/Tags }
using { /UnrealEngine.com/Temporary/Diagnostics }
using { /UnrealEngine.com/Temporary/UI }
using { /Fortnite.com/UI }

#=======================================================================================================================
#Classes

    booster_upgrade_data:=class<concrete>():
        @editable
        RarityIndex:int=0
        @editable
        LifeSpawn:float=0.0 #In seconds, use -1 for infinity duration
        @editable
        ExecutionsLimit:int=0 #-1 for infinity uses
        @editable
        BoosterDescriptionA:string="Information about booster " #E.g: Descript. A + {SomeData} + Descript. B
        @editable
        BoosterDescriptionB:string=" Additional information"
        @editable
        UpgradePrice:float=0.0 #for shop 
        # @editable #TODO: @Cbray, 30/01/2025, Maybe this should be implemented in an interface
        # UpgradeCurrecy:int=0 #for shop interface (Gold, Diamond, Custom)

  #-----------------------------------------------------------------------------------------------------------------------
    booster_instance_attributes:=class:
        var BoosterUpgradeIndex:int=0

    (BoosterInstanceAttributes:booster_instance_attributes).SetBoosterUpgradeIndex(NewUpgradeIndex:int):void=
        set BoosterInstanceAttributes.BoosterUpgradeIndex = NewUpgradeIndex
  #-----------------------------------------------------------------------------------------------------------------------
  #Every booster instance will store it's own data, like upgrade level, or other required information 
    booster_instance_attributes_data:=class:
        var Data:[]booster_instance_attributes=array{}
        var bIsRunning:logic=false

    (BoosterInstanceAttributesData:booster_instance_attributes_data).AddBoosterInstance(BoosterInstanceAttributes:booster_instance_attributes):void=
        set BoosterInstanceAttributesData.Data += array{BoosterInstanceAttributes}

    (BoosterInstanceAttributesData:booster_instance_attributes_data).RemoveBoosterInstance(BoosterInstanceAttributes:booster_instance_attributes):void=
        if(NewData:= BoosterInstanceAttributesData.Data.RemoveElement[0]): #@CBray 01/10/2025,maybe can be simplify to a single Line, check if the array is direct ref or copy
            set BoosterInstanceAttributesData.Data = NewData

    (BoosterInstanceAttributesData:booster_instance_attributes_data).SetIsRunning(IsRunning:logic):void=
        set BoosterInstanceAttributesData.bIsRunning = IsRunning
  #-----------------------------------------------------------------------------------------------------------------------
    booster_instances:=class:
        var BoostersPool:[booster_id]booster_instance_attributes_data=map{}
        var PlayerCanvas:canvas=canvas{}
        var PlayerStackBox:stack_box=stack_box{Orientation:=orientation.Horizontal}
        var bIsCanvasCreated:logic=false

    (BoosterIntaces:booster_instances).AddInstance(BoosterID:booster_id, BoosterUpgradeIndex:int):void=

        #Create new data
        New_BoosterInstanceAttributes:booster_instance_attributes=booster_instance_attributes{BoosterUpgradeIndex:=BoosterUpgradeIndex} #Attributes for this booster

        #if not exist, create new
        if(not BoosterIntaces.BoostersPool[BoosterID]):
            BoosterInstanceAttributesData:booster_instance_attributes_data=booster_instance_attributes_data{} #Create Booster Pool
            BoosterInstanceAttributesData.AddBoosterInstance(New_BoosterInstanceAttributes) #Add Booster to Pool by UpgradeIndex
            if(set BoosterIntaces.BoostersPool[BoosterID] = BoosterInstanceAttributesData){} #Create Booster Map by ID
            return

        #Update Boosters Pool
        if:
            Target_BoosterInstanceAttributesData:=BoosterIntaces.BoostersPool[BoosterID]
        then:
            Target_BoosterInstanceAttributesData.AddBoosterInstance(New_BoosterInstanceAttributes)

    (BoosterIntaces:booster_instances).RemoveInstance(BoosterID:booster_id):int=
        if:
            DataArray:=BoosterIntaces.BoostersPool[BoosterID].Data
        then:
            if(NewDataArray:[]booster_instance_attributes=DataArray.RemoveElement[0]):
                if(set BoosterIntaces.BoostersPool[BoosterID].Data = NewDataArray){}
                return NewDataArray.Length
        return 0
        
    (BoosterIntaces:booster_instances).RemoveAllInstances(BoosterID:booster_id):void=
        if(set BoosterIntaces.BoostersPool[BoosterID].Data = array{}){}

    (BoosterIntaces:booster_instances).GetNumberOfIntances(BoosterID:booster_id):int=
        if:
            DataArray:=BoosterIntaces.BoostersPool[BoosterID].Data
        then:
            return DataArray.Length
        return 0

    (BoosterIntaces:booster_instances).SetPlayerWidgetData(Canvas:canvas, BoostersStackBox:stack_box):void=
        set BoosterIntaces.PlayerCanvas = Canvas
        set BoosterIntaces.PlayerStackBox = BoostersStackBox
        set BoosterIntaces.bIsCanvasCreated = true

    (BoosterIntaces:booster_instances).ResetPlayerWidgetData():void=
        set BoosterIntaces.PlayerCanvas = canvas{}
        set BoosterIntaces.PlayerStackBox = stack_box{Orientation:=orientation.Horizontal}
        set BoosterIntaces.bIsCanvasCreated = false

    (BoosterIntaces:booster_instances).SetBoosterIsRunning(BoosterID:booster_id, IsRunning:logic):void=
        if:
            Target_BoosterInstanceAttributesData:=BoosterIntaces.BoostersPool[BoosterID]
        then:
            Target_BoosterInstanceAttributesData.SetIsRunning(IsRunning)

    (BoosterIntaces:booster_instances).GetBoosterIsRunning(BoosterID:booster_id):logic=
        if:
            Target_BoosterInstanceAttributesData:=BoosterIntaces.BoostersPool[BoosterID]
        then:
            return Target_BoosterInstanceAttributesData.bIsRunning
        return false
  #-----------------------------------------------------------------------------------------------------------------------
    booster_run_data:=class:
        var LifeSpawnLeft:float=0.0
        var ExecutionsLeft:int=0

    (BoosterRunData:booster_run_data).UpdateLifeSpawn(LifeSpawn:float):void=
        set BoosterRunData.LifeSpawnLeft=LifeSpawn

    (BoosterRunData:booster_run_data).UpdateExecutionsLimit(ExecutionsLimit:int):void=
        set BoosterRunData.ExecutionsLeft=ExecutionsLimit
  #-----------------------------------------------------------------------------------------------------------------------
    booster_base_widget_data:=struct:
        Parent:widget
        IconTextureBlock:texture_block
        IconOverlay:overlay
        DurationTextBlock:text_block
        AmountTextBlock:text_block
        InstancesTextureBlock:texture_block
        InstancesOverlay:overlay

    (BoosterBaseWidgetData:booster_base_widget_data).RemoveParentWidget():void=
        if (Parent:=BoosterBaseWidgetData.Parent.GetParentWidget[]):
            if(ParentStackBox:=stack_box[Parent]):
                ParentStackBox.RemoveWidget(BoosterBaseWidgetData.Parent)

#=========================================================================================================================

#Enum

    booster_id:=enum:
        None,
        SwiftFeet,
        FireBullets,
        HealthRegen,
        ShieldRegen,
        MaxHealth,
        MaxShield,
        ShieldAdrenaline,
        StrongComeback,
        VampireBullets,
        ExplosiveEnemies,
        HealthyDrops,
        GoldDigger,
        MysteryItem
        #Delete theese TEST members when implement yours
  #-----------------------------------------------------------------------------------------------------------------------
    booster_id_array:[]booster_id=array{
        booster_id.SwiftFeet
        booster_id.FireBullets
        booster_id.HealthRegen
        booster_id.ShieldRegen
        booster_id.MaxHealth
        booster_id.MaxShield
        booster_id.ShieldAdrenaline
        booster_id.StrongComeback
        booster_id.VampireBullets
        booster_id.ExplosiveEnemies
        booster_id.HealthyDrops
        }
        #Delete theese TEST members when implement yours
  #-----------------------------------------------------------------------------------------------------------------------
    # BoosterIDToDisplayName(BoosterID:booster_id):string=
    #     case (booster_id):
    #         booster_id.Blank =>""
    #     #Delete theese TEST members when implement yours
#=========================================================================================================================

#Tags
#Use this tag to detect all booster device in world
    booster_device_tag:=class(tag){}

#=========================================================================================================================

#Tools
#Comment methods you had already implemented in you own project

#     StringToMessage<localizes><public>(String:string)<computes>:message = "{String}"
#     AgentToMessage<localizes><public>(Agent:agent)<computes>:message = "{Agent}"
#   #-----------------------------------------------------------------------------------------------------------------------
#     TimerToString(Minutes:int, Seconds:int)<computes>:string =
#         # If Seconds < 10, then add a 0 in front so it displays 0# instead of #
#         SecondsAsString := if (Seconds < 10) then Join(array{ToString(0), ToString(Seconds)}, "") else ToString(Seconds)
#         "{Minutes}:{SecondsAsString}"
#   #-----------------------------------------------------------------------------------------------------------------------  
#     TimeInSecondsToString(Seconds:int)<computes>:string =
#         TimerToString(SecondsToMinutes(Seconds), GetRemainderSecondsFromSecondsToMinutes(Seconds))
#   #-----------------------------------------------------------------------------------------------------------------------   
#     SecondsToMinutes(Seconds:int)<computes>:int =
#         Floor(Seconds / 60 or Err())
#   #-----------------------------------------------------------------------------------------------------------------------
#     # Returns the reminder from seconds when changing to minutes
#     GetRemainderSecondsFromSecondsToMinutes(Seconds:int)<computes>:int =
#         Mod[Seconds, 60] or 0
#   #-----------------------------------------------------------------------------------------------------------------------    
#     DurationToString(Seconds:int)<computes>: string =
#         if (Minutes:=SecondsToMinutes(Seconds), Minutes > 0):
#             return "{Minutes}min"
#         return "{Seconds}s"
  #------------------------------------------------------------------------------------------------------------------------
#       <#
#     Returns the index of the first occurrence of SearchChar in Str.
#     Fails if SearchChar isn't found.
#     The function must be <transacts> for now to mutate local data (FoundIndex).
#     #>
#     FindCharIndex(SearchChar:char, Str:[]char)<transacts><decides>:int=
#         var FoundIndex:int = 0
#         var Found:logic = false
#         for (X -> Char : Str, not Found?):
#             if (Char = SearchChar):
#                 set FoundIndex = X
#         FoundIndex
#   #------------------------------------------------------------------------------------------------------------------------
#     # Returns Number converted to a string, truncated to Decimals places. 
#     ToTruncatedString(Number:float, Decimals:int):[]char=
#         var Str:[]char = ToString(Number)
#         if:
#             DotIndex := FindCharIndex['.', Str]
#             StopIndex := if (Decimals > 0) then Min(DotIndex+Decimals+1,Str.Length) else DotIndex
#             SlicedStr := Str.Slice[0, StopIndex] 
#             set Str = SlicedStr
#         Str
#=========================================================================================================================
#Widget Utils

    BlinkWidget(TargetWidget:widget, Blinks:int, TerminateBlinkEvent:event(),?BlinkDelay:float=0.5)<suspends>:void=
        var BlinkCount:int=0
        TargetWidget.SetVisibility(widget_visibility.Visible)
        race:
            TerminateBlinkEvent.Await()
            block:
                defer:
                    # Print("BlinkWidget: Defer Triggered!")
                    TargetWidget.SetVisibility(widget_visibility.Visible)
                loop:
                    CurrentVisibility:=TargetWidget.GetVisibility()

                    if(CurrentVisibility = widget_visibility.Visible):
                        TargetWidget.SetVisibility(widget_visibility.Hidden)
                    else:
                        TargetWidget.SetVisibility(widget_visibility.Visible)
                        set BlinkCount+=1

                    Sleep(BlinkDelay)

                    if(BlinkCount >=Blinks):
                        TargetWidget.SetVisibility(widget_visibility.Visible)
                        break

#=========================================================================================================================

#Debug
    log_booster_manager:=class(log_channel){}

#=========================================================================================================================

#Images (Importing modules)
    BoosterSystem<public> := module:
        Textures<public> := module:
            Icons<public> := module:
                Legacy<public> := module{}
#=========================================================================================================================