
using { /Fortnite.com/Devices }
using { /Fortnite.com/Characters }
using { /Fortnite.com/UI }
using { /Verse.org/Simulation }
using { /Verse.org/Assets }
using { /Verse.org/Concurrency }
using { /Verse.org/Colors }
using { /Verse.org/Colors/NamedColors }
using { /UnrealEngine.com/Temporary/Diagnostics }
using { /UnrealEngine.com/Temporary/UI }
using { /UnrealEngine.com/Temporary/SpatialMath }

#New Booster Devices are required to have a Tag Component and set the @booster_manager_device->SearchingTag tag.

#==============================================================================================================================

# Booster Upgrade Data Default Values
<# Create a Struct and inside create the array with the default values for the booster#>

# booster_default_upgrade_data:=struct:

#     DefaultValues:[]booster_upgrade_data=array{
#         Upgrade0:booster_upgrade_data=booster_upgrade_data{LifeSpawn:=10.0, ExecutionsLimit:=0}
#         Upgrade1:booster_upgrade_data=booster_upgrade_data{LifeSpawn:=15.0, ExecutionsLimit:=0}
#         Upgrade2:booster_upgrade_data=booster_upgrade_data{LifeSpawn:=20.0, ExecutionsLimit:=0}
#     }

#==============================================================================================================================

booster_device := class(creative_device):

    Logger:log=log{Channel:=log_booster_manager, DefaultLevel:=log_level.Debug}

    @editable
    bDebugEnabled:logic=false
    @editable
    bEnabled:logic=true
    BoosterID:booster_id=booster_id.None
    @editable
    BoosterDisplayName:string="Booster Name"
    #This only works if the change is done by verse, if you change on editor, you'll get a Blank Squared
    BoosterIcon:texture=BoosterSystem.Textures.Icons.T_UI_BSTR_Blank
    bIsStackable:logic=false #this means if you run an other instance of this booster, they can sum effects
    @editable #Use this to indicate there's more booster intances
    bDisplaySmallIconInsteadOfNumber:logic=true
    #Create this on your new class
    #var BoosterUpgradeData:[]booster_upgrade_data=booster_default_upgrade_data{}.DefaultValues

    #Events
    # int (UpgradeDataIndex)
    OnBoosterAddedEvent:event(tuple(player, booster_id, int))=event(tuple(player, booster_id, int)){}
    OnBoosterStartedEvent:event(tuple(player, booster_id, int))=event(tuple(player, booster_id, int)){}
    OnBoosterLoopCompletedEvent:event(tuple(player, booster_id))=event(tuple(player, booster_id)){}
    #@logic: Was Player Eliminated
    OnBoosterFinishedEvent:event(tuple(player, booster_id, logic))=event(tuple(player, booster_id, logic)){}
    OnBoosterSumEffectEvent:event(tuple(player, booster_id))=event(tuple(player, booster_id)){}

    #@Player, booster_id, int (Instances), int (UpgradeDataIndex)
    OnBoosterNewInstanceAddedEvent:event(tuple(player, booster_id, int, int))=event(tuple(player, booster_id, int, int)){}
    OnBoosterInstanceRemovedEvent:event(tuple(player, booster_id, int))=event(tuple(player, booster_id, int)){}
    #@Player, booster_id, int (UpgradeDataIndex)
    OnBoosterNewInstanceStartedEvent:event(tuple(player, booster_id, int))=event(tuple(player, booster_id, int)){}
    OnBoosterRunDataUpdatedEvent:event(booster_run_data)=event(booster_run_data){}
#==============================================================================================================================
    OnBegin<override>()<suspends>:void=
        if(bEnabled?):

#==============================================================================================================================
    GetBaseBoosterUpgradeData(BoosterUpgradeIndex:int)<decides><transacts>:booster_upgrade_data=booster_upgrade_data{}
        # TargetUpgradeData:=BoosterUpgradeData[BoosterUpgradeIndex]
        #Implement this on your new Class
#============================================================================================================================== 
    GetBoosterDescription(BoosterUpgradeIndex:int):string=""
     #Implement this on your new Class
#============================================================================================================================== 
    GetMaxBoosterUpgradeIndex():int=0
     #Implement this on your new Class
#============================================================================================================================== 
    AddBoosterToPlayer(Player:player, BoosterUpgradeIndex:int, ?IsAdditionalInstance:logic=false):void=
        spawn{AwaitForBoosterStart(Player, BoosterUpgradeIndex, ?IsAdditionalInstance:=IsAdditionalInstance)}
        OnBoosterAddedEvent.Signal(Player, BoosterID, BoosterUpgradeIndex)
#==============================================================================================================================        
    AwaitForBoosterStart(Player:player, BoosterUpgradeIndex:int, ?IsAdditionalInstance:logic=false)<suspends>:void=
        <#here fill the events for requirements you need to start the Booster, E.g: on Player PickUp an item or kill an enemy, etc.
         Call this as "(super:)" at the end of your code#>
        StartBoosterOnPlayer(Player, BoosterUpgradeIndex, ?IsAdditionalInstance:=IsAdditionalInstance)
#============================================================================================================================== 
    StartBoosterOnPlayer<protected>(Player:player, BoosterUpgradeIndex:int, ?IsAdditionalInstance:logic=false):void=
        if(bDebugEnabled?):
            Logger.Print("Starting Booster: {BoosterDisplayName}, Upgrade Data Index: {BoosterUpgradeIndex}")

        if(not bIsStackable? or bIsStackable? and not IsAdditionalInstance?):
            spawn{BoosterHandler(Player,BoosterUpgradeIndex)}
            OnBoosterStarted(Player, BoosterUpgradeIndex)

        if(IsAdditionalInstance?):
            OnBoosterNewInstanceStartedEvent.Signal(Player, BoosterID, BoosterUpgradeIndex)
        
#==============================================================================================================================  
    StopBoosterForPlayer(Player:player):void=
        OnBoosterFinished(Player)
#==============================================================================================================================
    OnBoosterStarted<private>(Player:player, BoosterUpgradeIndex:int):void=
     OnBoosterStartedEvent.Signal(Player, BoosterID, BoosterUpgradeIndex)
#==============================================================================================================================
    BoosterLoopBehavior(Player:player, BoosterUpgradeIndex:int, BoosterLoopCompleteEvent:event(player), TerminateBoosterHandlerEvent:event())<suspends>:void=
    <#
      @Player: Player running this Booster
      @BoosterLoopCompleteEvent: if Loop is Complete Call the event to indicate booster was used
      @TerminateBoosterHandlerEvent.Await(): to finish this method, this is call from the main handler, if time is out or the execution limit was reached
      @TerminateLoopBehaviorEvent.Await() : To Finish this method and all sub rutines
      #>


      #Override this method in your class, use this template as guide

      TerminateLoopBehaviorEvent:event()=event(){}

      race:
        TerminateLoopBehaviorEvent.Await()
        sync:
            loop:
                TerminateBoosterHandlerEvent.Await()
                Logger.Print("{BoosterDisplayName}: Loop Finished", ?Level:=log_level.Warning)
                TerminateLoopBehaviorEvent.Signal()
            loop:
                EventData:=OnBoosterNewInstanceStartedEvent.Await()
                EventPlayer:=EventData(0)
                EventBoosterID:=EventData(1)  
                EventBoosterUpgradeIndex:=EventData(2) 
                if(EventPlayer = Player):
                    if(bIsStackable?):
                        <#Here you can increse the values of the booster, e.g: Max health, you can sum additional health to the same running booster
                        instead of starting a new corutine and have two of them#>
            loop:
                Sleep(0.0)
                BoosterLoopCompleteEvent.Signal(Player)
      
#==============================================================================================================================
    OnBoosterFinished(Player:player, ?WasPlayerEliminated:logic=false):void=
      OnBoosterFinishedEvent.Signal(Player, BoosterID, WasPlayerEliminated)
#==============================================================================================================================
    OnPlayerEliminated<private>(Player:player, TerminateEvent:event()):void=
     TerminateEvent.Signal()
     OnBoosterFinished(Player, ?WasPlayerEliminated:=true)
#==============================================================================================================================
    BoosterHandler<private>(Player:player, BoosterUpgradeIndex:int)<suspends>:void=
    
        if(BaseBoosterUpgradeData:=GetBaseBoosterUpgradeData[BoosterUpgradeIndex]):
 
            var BoosterRunData:booster_run_data=booster_run_data{
                LifeSpawnLeft:=BaseBoosterUpgradeData.LifeSpawn, 
                ExecutionsLeft:=BaseBoosterUpgradeData.ExecutionsLimit}

            #Events
            #Triggers when a single loop is completed
            BoosterLoopCompleteEvent:event(player)=event(player){}
            #Terminate this corutine
            TerminateBoosterHandlerEvent:event()=event(){}

            #Corutine that handles every Booster Logic
            spawn{BoosterLoopBehavior(Player, BoosterUpgradeIndex, BoosterLoopCompleteEvent, TerminateBoosterHandlerEvent)}

            race:
                TerminateBoosterHandlerEvent.Await()

                sync:

                    loop:
                        if(FortCharacter:=Player.GetFortCharacter[]):
                            FortCharacter.EliminatedEvent().Await()
                            OnPlayerEliminated(Player,TerminateBoosterHandlerEvent) #Event is Pass, if Create Custom behavior if required
                    loop:
                        EventData:=OnBoosterFinishedEvent.Await()
                        SentPlayer:=EventData(0)
                        if(SentPlayer = Player):
                            TerminateBoosterHandlerEvent.Signal()

                    loop:
                        EventData:=OnBoosterNewInstanceAddedEvent.Await()
                        SentPlayer:=EventData(0)
                        if(SentPlayer = Player):
                            if(bIsStackable?):
                                #Other Logic Here :)
                    loop:
                        EventData:=OnBoosterNewInstanceStartedEvent.Await()
                        SentPlayer:=EventData(0)
                        if(SentPlayer = Player):
                           OnBoosterNewInstanceStarted(Player, BoosterUpgradeIndex, BoosterRunData)

                    loop: #Finish this booster by time
                        Sleep(1.0)
                        if:
                            BaseBoosterUpgradeData.LifeSpawn >= 0.0
                            BaseBoosterUpgradeData.ExecutionsLimit = 0 
                        then:
                            #Booster LifeSpawn
                            set BoosterRunData.LifeSpawnLeft-=1.0
                            if(BoosterRunData.LifeSpawnLeft <=0.0):
                                Logger.Print("{BoosterDisplayName}: LifeSpawn Finished!", ?Level:=log_level.Warning)
                                StopBoosterForPlayer(Player)
                        else:
                            break
                    loop:
                        BoosterLoopCompleteEvent.Await() #Wait until booster loop finish and check the Times this booster can be used
                        OnBoosterLoopCompletedEvent.Signal(Player, BoosterID)
                        if:
                            BaseBoosterUpgradeData.LifeSpawn = 0.0 
                            BaseBoosterUpgradeData.ExecutionsLimit >= 0 
                        then:
                            set BoosterRunData.ExecutionsLeft-=1 #reduce the remaining executions      
                            TotalExecutions:=BaseBoosterUpgradeData.ExecutionsLimit-BoosterRunData.ExecutionsLeft
                            Logger.Print("{BoosterDisplayName}: Executions: {TotalExecutions}", ?Level:=log_level.Warning)  

                            if(BoosterRunData.ExecutionsLeft <= 0):
                                Logger.Print("{BoosterDisplayName}: Executions Completed: {TotalExecutions}", ?Level:=log_level.Warning) 
                                StopBoosterForPlayer(Player)
                        else:
                            break
        else:
            Logger.Print("{BoosterDisplayName}: Booster Upgrade Data not found", ?Level:=log_level.Error)          

#==============================================================================================================================
    GenerateBoosterWidget(Player:player, IconSize:vector2, BoosterUpgradeIndex:int, Canvas:canvas):stack_box=

        BoosterStackBox:=stack_box{Orientation:=orientation.Vertical}   
        SmallIconSize:=IconSize*0.72
         
        IconTextureBlock:texture_block=texture_block{DefaultImage:=BoosterIcon, DefaultDesiredSize:=IconSize}
        AmountTextBlock:text_block=text_block{DefaultTextColor:=White}
        DurationTextBlock:text_block=text_block{DefaultTextColor:=White}
        InstancesTextureBlock:texture_block=texture_block{DefaultImage:=BoosterIcon, DefaultDesiredSize:=SmallIconSize}
        
        if(not bIsStackable?):
            InstanceTintColor:=color{R:=255.0}
            InstancesTextureBlock.SetTint(InstanceTintColor)

        #SmallIcon Overlay
        InstancesOverlay:=overlay{
            Slots:=array{
                overlay_slot{
                    Widget:=InstancesTextureBlock
                    HorizontalAlignment:=horizontal_alignment.Center
                    VerticalAlignment:=vertical_alignment.Top
                }
            }
        }

        #Icon Overlay
        IconOverlay:=overlay{
            Slots:=array{
                overlay_slot{
                    Widget:=IconTextureBlock
                }
            }
        }

        Slots:=array{
            BlankSpacer:=stack_box_slot{ #Use it to Keep the StackBox
            Widget:=text_block{DefaultText:=StringToMessage("           ")}
            HorizontalAlignment:=horizontal_alignment.Left
            VerticalAlignment:=vertical_alignment.Top
            },
            Icon_Amount:=stack_box_slot{
                Widget:=stack_box{
                    Orientation:=orientation.Horizontal
                    Slots:=array{
                        IconTexture:=stack_box_slot{
                            Widget:=IconOverlay
                            HorizontalAlignment:=horizontal_alignment.Fill
                            VerticalAlignment:=vertical_alignment.Fill
                            Distribution:=option{1.0}
                        }
                        InstancesIcon:=stack_box_slot{
                            Widget:=InstancesOverlay
                            HorizontalAlignment:=horizontal_alignment.Left
                            VerticalAlignment:=vertical_alignment.Fill
                        }
                        AmountText:=stack_box_slot{
                            Widget:=AmountTextBlock
                            HorizontalAlignment:=horizontal_alignment.Left
                            VerticalAlignment:=vertical_alignment.Top
                        }
                    }
                }
                Padding:=margin{}             
                HorizontalAlignment:=horizontal_alignment.Fill
                VerticalAlignment:=vertical_alignment.Fill
            },
            DurationText:=stack_box_slot{
                Widget:=DurationTextBlock
                Padding:=margin{}             
                HorizontalAlignment:=horizontal_alignment.Center
                VerticalAlignment:=vertical_alignment.Center
            }
        }

        for (Slot:Slots):
            BoosterStackBox.AddWidget(Slot)

        AmountTextBlock.SetVisibility(widget_visibility.Collapsed)
        InstancesOverlay.SetVisibility(widget_visibility.Collapsed)

        BoosterWidgetData:=booster_base_widget_data{
            Parent:=BoosterStackBox
            IconTextureBlock:=IconTextureBlock
            IconOverlay:=IconOverlay
            DurationTextBlock:=DurationTextBlock
            AmountTextBlock:=AmountTextBlock
            InstancesTextureBlock:=InstancesTextureBlock
            InstancesOverlay:=InstancesOverlay
        }

        spawn{BoosterWidgetHandler(Player, BoosterWidgetData,BoosterUpgradeIndex)}

        return BoosterStackBox
#==============================================================================================================================
    BoosterWidgetHandler(Player:player, BoosterWidgetData:booster_base_widget_data, InitBoosterUpgradeIndex:int)<suspends>:void=
        
        CloseWidgetEvent:event()=event(){}
        #int Upgrade Data Index
        RefreshWidgetDataEvent:event(int)=event(int){}

        TerminateWidgetBlinkEvent:event()=event(){}

        var TargetLifeSpawn:float=0.0
        var TargetExecutionsLimit:int=0
        var bIsTimeControlled:logic=false
        var bIsInfinity:logic=false

        if(UpgradeData:=GetBaseBoosterUpgradeData[InitBoosterUpgradeIndex]):
            set TargetLifeSpawn = UpgradeData.LifeSpawn
            set TargetExecutionsLimit = UpgradeData.ExecutionsLimit

        if(TargetLifeSpawn=-1.0 or TargetExecutionsLimit = -1):
            set bIsInfinity = true

        if(TargetLifeSpawn>0.0):
            set bIsTimeControlled = true
            if(not bIsInfinity?):
                if(RoundedTime:=Round[TargetLifeSpawn]):
                    FormatedTime:= TimeInSecondsToString(RoundedTime)
                    ExecutionValue:=StringToMessage("{FormatedTime}")
                    BoosterWidgetData.DurationTextBlock.SetText(ExecutionValue)
        else:
            if(not bIsInfinity? and TargetExecutionsLimit>0):
                ExecutionValue:=StringToMessage("{TargetExecutionsLimit}")
                BoosterWidgetData.DurationTextBlock.SetText(ExecutionValue)

        race:
            CloseWidgetEvent.Await()
            sync:
                loop: #TODO: @Cbray: Mix start events
                    EventData:=OnBoosterStartedEvent.Await()
                    EventPlayer:=EventData(0)
                    EventBoosterID:=EventData(1)  
                    EventBoosterUpgradeIndex:=EventData(2) 
                    if(EventPlayer = Player):
                        if(not bIsInfinity? and not bIsStackable?):
                            if(bDebugEnabled?){ Logger.Print("Updating Widget Instance")}
                            RefreshWidgetDataEvent.Signal(EventBoosterUpgradeIndex)
                loop:
                    EventData:=OnBoosterNewInstanceStartedEvent.Await()
                    EventPlayer:=EventData(0)
                    EventBoosterID:=EventData(1)  
                    EventBoosterUpgradeIndex:=EventData(2) 
                    if(EventPlayer = Player):
                        if(not bIsInfinity? and not bIsStackable?):
                            if(bDebugEnabled?){ Logger.Print("Updating Widget Instance")}
                            RefreshWidgetDataEvent.Signal(EventBoosterUpgradeIndex)

                loop:
                    EventData:=OnBoosterFinishedEvent.Await()
                    EventPlayer:=EventData(0)
                    EventBoosterID:=EventData(1)               
                    if(EventPlayer = Player):
                        #Other Logic Here :)
                loop:
                    EventData:=OnBoosterNewInstanceAddedEvent.Await()
                    EventPlayer:=EventData(0)
                    EventBoosterID:=EventData(1)
                    EventNewInstances:=EventData(2)

                    if(EventPlayer = Player):
                        #Check which instances indicator method was set
                        var TargetWidget:widget=text_block{}
                        if(bDisplaySmallIconInsteadOfNumber?):
                            set TargetWidget = BoosterWidgetData.InstancesOverlay
                        else:
                            set TargetWidget = BoosterWidgetData.AmountTextBlock
                            Newvalue:=StringToMessage("{EventNewInstances}")
                            BoosterWidgetData.AmountTextBlock.SetText(Newvalue)

                        CurrentVisibility:=TargetWidget.GetVisibility()
                        if(not CurrentVisibility = widget_visibility.Visible):
                            TargetWidget.SetVisibility(widget_visibility.Visible)
                        
                        TerminateWidgetBlinkEvent.Signal()
                        spawn{BlinkWidget(TargetWidget, 4, TerminateWidgetBlinkEvent)}
                        
                loop:
                    EventData:=OnBoosterInstanceRemovedEvent.Await()
                    EventPlayer:=EventData(0)
                    EventBoosterID:=EventData(1)
                    EventRemainingInstances:=EventData(2)

                    if(EventPlayer = Player):
                        #Check which instances indicator method was set
                        var TargetWidget:widget=text_block{}
                    
                        if(bDisplaySmallIconInsteadOfNumber?):
                            set TargetWidget = BoosterWidgetData.InstancesOverlay
                        else:
                            set TargetWidget = BoosterWidgetData.AmountTextBlock
                            Newvalue:=StringToMessage("{EventRemainingInstances}")
                            BoosterWidgetData.AmountTextBlock.SetText(Newvalue)

                        CurrentVisibility:=TargetWidget.GetVisibility()
                        if:
                            CurrentVisibility = widget_visibility.Visible
                            EventRemainingInstances = 1
                        then:
                            TargetWidget.SetVisibility(widget_visibility.Collapsed)

                        if(EventRemainingInstances = 0):
                            if(bDebugEnabled?){Logger.Print("Closing Widget")}  
                            #Remove this Booster Widget
                            BoosterWidgetData.RemoveParentWidget()
                            if(PlayerUI:=GetPlayerUI[Player]):
                                PlayerUI.RemoveWidget(BoosterWidgetData.Parent)

                            #Finish the whole Method
                            CloseWidgetEvent.Signal()
                        else: 
                            #Make widget to blink when an instance is removed
                            IconWidget:=BoosterWidgetData.IconTextureBlock
                            TerminateWidgetBlinkEvent.Signal()
                            spawn{BlinkWidget(IconWidget, 4, TerminateWidgetBlinkEvent)}
                loop:
                    if(bIsInfinity? or TargetLifeSpawn <= 0.0 and not bIsTimeControlled?): #if -1.0, infinity, so this loop is not required
                        if(bDebugEnabled?){Logger.Print("LifeSpawn Count UI Finished!")}  
                        break
                    Sleep(1.0)
                    set TargetLifeSpawn-=1.0
                    var FormatedTime:string=""
                    if(RoundedTime:=Round[TargetLifeSpawn]):
                        set FormatedTime = TimeInSecondsToString(RoundedTime) 
                    DurationValue:=StringToMessage("{FormatedTime}")
                    BoosterWidgetData.DurationTextBlock.SetText(DurationValue)
                loop:
                    if(bIsInfinity? or TargetExecutionsLimit <= 0 and bIsTimeControlled?): #if -1, infinity, so this loop is not required
                        if(bDebugEnabled?){Logger.Print("Executions Limit Count UI Finished!")}  
                        break
                    EventData:=OnBoosterLoopCompletedEvent.Await()
                    EventPlayer:=EventData(0)
                    if(EventPlayer = Player):
                        set TargetExecutionsLimit -=1
                        ExecutionValue:=StringToMessage("{TargetExecutionsLimit}")
                        BoosterWidgetData.DurationTextBlock.SetText(ExecutionValue)
                loop:
                    NewUpgradeDataIndex:=RefreshWidgetDataEvent.Await()
                    if(UpgradeData:=GetBaseBoosterUpgradeData[NewUpgradeDataIndex]):
                        if(UpgradeData.LifeSpawn > 0.0):
                            set TargetLifeSpawn = UpgradeData.LifeSpawn
                            if(RoundedTime:=Round[TargetLifeSpawn]):
                                FormatedTime := TimeInSecondsToString(RoundedTime)
                                DurationValue:=StringToMessage("{FormatedTime}")
                                BoosterWidgetData.DurationTextBlock.SetText(DurationValue)
                        else:
                            if(UpgradeData.ExecutionsLimit > 0):
                                set TargetExecutionsLimit = UpgradeData.ExecutionsLimit
                                ExecutionValue:=StringToMessage("{TargetExecutionsLimit}")
                                BoosterWidgetData.DurationTextBlock.SetText(ExecutionValue)
                loop: #This needs to be mixed into one single method
                    EventData:=OnBoosterRunDataUpdatedEvent.Await()
                    if(EventData.LifeSpawnLeft > 0.0):
                        set TargetLifeSpawn = EventData.LifeSpawnLeft
                        if(RoundedTime:=Round[TargetLifeSpawn]):
                            FormatedTime := TimeInSecondsToString(RoundedTime)
                            DurationValue:=StringToMessage("{FormatedTime}")
                            BoosterWidgetData.DurationTextBlock.SetText(DurationValue)
                    else:
                        if(EventData.ExecutionsLeft > 0):
                            set TargetExecutionsLimit = EventData.ExecutionsLeft
                            ExecutionValue:=StringToMessage("{TargetExecutionsLimit}")
                            BoosterWidgetData.DurationTextBlock.SetText(ExecutionValue)
                
#============================================================================================================================
#Override this with your Booster Upgrade data if you need to clamp some values
    OnBoosterNewInstanceStarted(Player:player, NewBoosterUpgradeIndex:int, CurrentBoosterRunData:booster_run_data):void=
        if(BaseBoosterUpgradeData:=GetBaseBoosterUpgradeData[NewBoosterUpgradeIndex]):
            if(bIsStackable?):
                Print("Updating Run Data")
                LifeSpawn:=BaseBoosterUpgradeData.LifeSpawn
                CurrentBoosterRunData.UpdateLifeSpawn(LifeSpawn)    
                #  or   
                ExecutionsLimit:=BaseBoosterUpgradeData.ExecutionsLimit
                CurrentBoosterRunData.UpdateExecutionsLimit(ExecutionsLimit) 

                OnBoosterRunDataUpdatedEvent.Signal(CurrentBoosterRunData)


