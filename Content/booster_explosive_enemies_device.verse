using { /Fortnite.com/Devices }
using { /Fortnite.com/Characters }
using { /Verse.org/Simulation }
using { /Verse.org/Assets }
using { /UnrealEngine.com/Temporary/Diagnostics }
using { /Fortnite.com/Game }
using { /UnrealEngine.com/Temporary/SpatialMath }

#Booster Set Up
#-----------------------------------------------------------------------------------------------------------------

booster_upgrade_data_explosive_enemies := class<concrete>(booster_upgrade_data):

    LifeSpawn<override>:float=300.0
    BoosterDescriptionA<override>:string="Enemies explode in "
    BoosterDescriptionB<override>:string="\n radius when you kill them"
    @editable
    ExplosionRadius:float = 250.0
    @editable
    VFXExplosion:vfx_spawner_device = vfx_spawner_device{}


booster_default_upgrade_data_explosive_enemies:=struct:

    DefaultValues:[]booster_upgrade_data_explosive_enemies = array{
        Upgrade0:booster_upgrade_data_explosive_enemies = booster_upgrade_data_explosive_enemies { }
        Upgrade1:booster_upgrade_data_explosive_enemies = booster_upgrade_data_explosive_enemies { ExplosionRadius:= 400.0 }
        Upgrade2:booster_upgrade_data_explosive_enemies = booster_upgrade_data_explosive_enemies { ExplosionRadius:= 550.0 }
        Upgrade3:booster_upgrade_data_explosive_enemies = booster_upgrade_data_explosive_enemies { ExplosionRadius:= 800.0 }
    }

#Booster Data
#-----------------------------------------------------------------------------------------------------------------

booster_explosive_enemies_device := class(booster_device):

    @editable
    var BoosterUpgradeData:[]booster_upgrade_data_explosive_enemies = booster_default_upgrade_data_explosive_enemies{}.DefaultValues

    @editable
    ExplosionDamage:float = 100.0
    @editable
    VFXOffsetLocation:vector3 = vector3{}
    @editable
    GuardSpawner:guard_spawner_device=guard_spawner_device{}

    BoosterID<override>:booster_id = booster_id.ExplosiveEnemies
    BoosterDisplayName<override>:string = "Explosive Enemies"
    BoosterIcon<override>:texture=BoosterSystem.Textures.Icons.Legacy.T_ExplosiveEnemies

#Booster Functions
#-----------------------------------------------------------------------------------------------------------------

    GetBoosterUpgradeData(BoosterUpgradeIndex:int)<decides><transacts>:booster_upgrade_data_explosive_enemies =
        BoosterUpgradeData[BoosterUpgradeIndex]
        
    GetBaseBoosterUpgradeData<override>(BoosterUpgradeIndex:int)<decides><transacts>:booster_upgrade_data=GetBoosterUpgradeData[BoosterUpgradeIndex]

    GetMaxBoosterUpgradeIndex<override>():int=BoosterUpgradeData.Length

    GetBoosterDescription<override>(BoosterUpgradeIndex:int):string= 
        if:
            BoosterData:=GetBoosterUpgradeData[BoosterUpgradeIndex]
        then:
            TruncatedString:=ToTruncatedString(BoosterData.ExplosionRadius,1)
            return BoosterData.BoosterDescriptionA + "{TruncatedString}" +BoosterData.BoosterDescriptionB
        return "Booster Description"

    BoosterLoopBehavior<override>(Player:player, BoosterUpgradeIndex:int, BoosterLoopCompleteEvent:event(player), TerminateBoosterHandlerEvent:event())<suspends>:void =

        if(TargetBoosterData:=GetBoosterUpgradeData[BoosterUpgradeIndex]):
        
            TerminateLoopBehaviorEvent:event()=event(){}
        
            if(FortCharacter := Player.GetFortCharacter[]):
                        
                race:
                    TerminateLoopBehaviorEvent.Await()
                    sync:
                        loop:
                            TerminateBoosterHandlerEvent.Await()
                            Logger.Print("{BoosterDisplayName}: Loop Finished", ?Level:=log_level.Warning)
                            TerminateLoopBehaviorEvent.Signal()
                        loop:
                            Sleep(1.0)
                            Logger.Print("{BoosterDisplayName}: Running, Upgrade Index: {BoosterUpgradeIndex}")
                            BoosterLoopCompleteEvent.Signal(Player)
                        loop:
                            #TODO Replace with Enemies Manager to detect when an NPC has died and get the agent
                            Result := GuardSpawner.EliminatedEvent.Await()
                            QAgent := Result.Target

                            if:
                                QAgent <> false
                                Agent: agent = QAgent?
                                EliminatedCharacter: fort_character = Agent.GetFortCharacter[]
                                EliminatedTransform := EliminatedCharacter.GetTransform()
                            then:
                                #TODO Get guards in room from enemy manager
                                #spawn{DamageEnemies(GuardsInRoom, BoosterUpgradeIndex, QAgent, FortCharacter)}
                                if (TargetBoosterData.VFXExplosion.TeleportTo[EliminatedTransform.Translation + VFXOffsetLocation, EliminatedTransform.Rotation]):
                                    TargetBoosterData.VFXExplosion.Restart()


    DamageEnemies(Guards:[]agent, BoosterUpgradeIndex:int, EliminatedGuard : agent, FortCharacter: fort_character)<suspends>: void =
        if:
            EliminatedCharacter: fort_character = EliminatedGuard.GetFortCharacter[]
            EliminatedTransform := EliminatedCharacter.GetTransform()
            TargetBoosterData:=GetBoosterUpgradeData[BoosterUpgradeIndex]
        then:
            for (Guard:Guards, Guard <> EliminatedGuard, GuardCharacter:=Guard.GetFortCharacter[]):
                # Validate if guard in the explosion radius
                SqrDistance := DistanceSquared(EliminatedTransform.Translation, GuardCharacter.GetTransform().Translation)
                if (SqrDistance <= (TargetBoosterData.ExplosionRadius * TargetBoosterData.ExplosionRadius)):
                    GuardCharacter.Damage(damage_args:
                        Instigator:=option{FortCharacter}
                        Source:=option{FortCharacter}
                        Amount:=ExplosionDamage
                    )
                    Sleep(0.2)