using { /Fortnite.com/Devices }
using { /Fortnite.com/Characters }
using { /Verse.org/Simulation }
using { /Verse.org/Assets }
using { /UnrealEngine.com/Temporary/Diagnostics }

booster_upgrade_data_fire_bullets:=class<concrete>(booster_upgrade_data):
    LifeSpawn<override>:float=300.0
    BoosterDescriptionA<override>:string="Set enemies on fire for\n"
    BoosterDescriptionB<override>:string=" seconds"
    @editable
    Damage:float = 1.0
    @editable
    DamageWindow:float=5.0
    @editable
    MaxDamageWindow:float=10.0

booster_fire_bullets_damage_data:=class():
    var AmountOfDamage:float=0.0
    var DamageWindow:float=0.0

(DamageData:booster_fire_bullets_damage_data).UpdateDamageWindow(NewTime:float):float=
    set DamageData.DamageWindow += NewTime 
    return DamageData.DamageWindow

(DamageData:booster_fire_bullets_damage_data).UpdateDamageData(NewDamage:float, NewTime:float, ?MaxTime:float= 10.0):tuple(float, float)=
    set DamageData.AmountOfDamage += NewDamage
    set DamageData.DamageWindow = Clamp(DamageData.DamageWindow+NewTime,0.0, MaxTime)  

    return (DamageData.AmountOfDamage, DamageData.DamageWindow)

booster_default_upgrade_data_fire_bullets:=struct:

    DefaultValues:[]booster_upgrade_data_fire_bullets=array{
        Upgrade0:booster_upgrade_data_fire_bullets=booster_upgrade_data_fire_bullets{}
        Upgrade1:booster_upgrade_data_fire_bullets=booster_upgrade_data_fire_bullets{Damage:=2.0, DamageWindow:=10.0, MaxDamageWindow:=20.0}
        Upgrade2:booster_upgrade_data_fire_bullets=booster_upgrade_data_fire_bullets{Damage:=3.0, DamageWindow:=20.0, MaxDamageWindow:=40.0}
        Upgrade3:booster_upgrade_data_fire_bullets=booster_upgrade_data_fire_bullets{Damage:=5.0, DamageWindow:=30.0, MaxDamageWindow:=60.0}
    }
#==============================================================================================================================
booster_fire_bullets_device := class(booster_device):

    @editable
    var BoosterUpgradeData:[]booster_upgrade_data_fire_bullets=booster_default_upgrade_data_fire_bullets{}.DefaultValues

    @editable
    GuardSpawner:guard_spawner_device=guard_spawner_device{}

    @editable
    VfxDevice:visual_effect_powerup_device = visual_effect_powerup_device{}

    @editable #Delay between damage applied to enemy
    DamageDelay:float=1.0

    var EnemyPool:[fort_character]booster_fire_bullets_damage_data=map{}

    var bIsDamageHandlerRunning:logic=false

    BoosterID<override>:booster_id=booster_id.FireBullets
    BoosterDisplayName<override>:string="Fire Bullets"
    BoosterIcon<override>:texture=BoosterSystem.Textures.Icons.Legacy.T_FireBullets

    #Events
    OnEnemyDamagedEvent:event(tuple(fort_character, ?fort_character))=event(tuple(fort_character, ?fort_character)){}

#==============================================================================================================
    OnBegin<override>()<suspends>:void=
        if(bEnabled?):
            #
            spawn{AwaitForEnemyDamage()}
#==============================================================================================================
    GetBoosterUpgradeData(BoosterUpgradeIndex:int)<decides><transacts>:booster_upgrade_data_fire_bullets=
        BoosterUpgradeData[BoosterUpgradeIndex]
#==============================================================================================================================       
    GetBaseBoosterUpgradeData<override>(BoosterUpgradeIndex:int)<decides><transacts>:booster_upgrade_data=GetBoosterUpgradeData[BoosterUpgradeIndex]
#==============================================================================================================================
    GetMaxBoosterUpgradeIndex<override>():int=BoosterUpgradeData.Length
#==============================================================================================================================
    GetBoosterDescription<override>(BoosterUpgradeIndex:int):string= 
        if:
            BoosterData:=GetBoosterUpgradeData[BoosterUpgradeIndex]
        then:
            TruncatedString:=ToTruncatedString(BoosterData.DamageWindow,1)
            return BoosterData.BoosterDescriptionA + "{TruncatedString}" +BoosterData.BoosterDescriptionB
        return "Booster Description"
#==============================================================================================================================
    BoosterLoopBehavior<override>(Player:player, BoosterUpgradeIndex:int, BoosterLoopCompleteEvent:event(player), TerminateBoosterHandlerEvent:event())<suspends>:void=
        if(TargetBoosterData:=GetBoosterUpgradeData[BoosterUpgradeIndex]):
            
            DamageToApply:=TargetBoosterData.Damage
            DamageWindow:=TargetBoosterData.DamageWindow
            MaxDamageWindow:=TargetBoosterData.MaxDamageWindow
            TerminateLoopBehaviorEvent:event()=event(){}

            race:
                TerminateLoopBehaviorEvent.Await()
                sync:
                    loop:
                        TerminateBoosterHandlerEvent.Await()
                        if(bDebugEnabled?){Logger.Print("{BoosterDisplayName}: Loop Terminated", ?Level:=log_level.Warning)}
                        TerminateLoopBehaviorEvent.Signal()
                    loop:
                        Result:=OnEnemyDamagedEvent.Await()

                        if(not bIsDamageHandlerRunning? and EnemyPool.Length=0):
                            if(bDebugEnabled?){Logger.Print("{BoosterDisplayName}: Starting Damage Handler")}
                            spawn{DamageHandler()}
                            set bIsDamageHandlerRunning = true

                        if:
                            TargetFC:=Result(0)
                            if(bDebugEnabled?){Logger.Print("Valid Target")}
                            SourceFC:=Result(1)?
                            if(bDebugEnabled?){Logger.Print("Valid Source")}
                            SourceAgent:=SourceFC.GetAgent[]
                            if(bDebugEnabled?){Logger.Print("Valid Agent")}
                            SourcePlayer:=player[SourceAgent]
                            if(bDebugEnabled?){Logger.Print("Valid Player")}
                            Player = SourcePlayer
                            if(bDebugEnabled?){Logger.Print("Player Match")}
                        then:
                            FortCharacter:=TargetFC
                            if(TargetEnemyData:= EnemyPool[FortCharacter]):
                                NewData:=TargetEnemyData.UpdateDamageData(DamageToApply, DamageWindow, ?MaxTime:=MaxDamageWindow)
                                if(bDebugEnabled?){Logger.Print("Damage:= {NewData(0)}, Window:= {NewData(1)}")}
                            else:
                                DamageData:=booster_fire_bullets_damage_data{}
                                set DamageData.AmountOfDamage =  DamageToApply
                                set DamageData.DamageWindow =  DamageWindow
                                AddEnemyToArray(FortCharacter, DamageData)
                                StartVfx(FortCharacter)
                                if(bDebugEnabled?){Logger.Print("Damage:= {DamageToApply}, Window:= {DamageWindow}")}

                        BoosterLoopCompleteEvent.Signal(Player)
#==============================================================================================================================
    StartVfx(Enemy:fort_character):void=
        if(Agent:=Enemy.GetAgent[]):
            VfxDevice.Spawn()
            VfxDevice.SetDuration(-1.0) #Infinity
            VfxDevice.Pickup(Agent)
#==============================================================================================================================
    StopVfx(Enemy:fort_character):void=
        if(Agent:=Enemy.GetAgent[]):
            VfxDevice.Spawn()
            VfxDevice.SetDuration(0.5) #For remove
            VfxDevice.Pickup(Agent)
 #==============================================================================================================================
    AddEnemyToArray(Enemy:fort_character, DamageData:booster_fire_bullets_damage_data):void=
        if:
            set EnemyPool[Enemy] = DamageData
#==============================================================================================================================
    DamageHandler()<suspends>:void=
        loop:
            Sleep(DamageDelay)       
            if(EnemyPool.Length = 0):
                set bIsDamageHandlerRunning = false
                break
            EnemyPoolCopy:=EnemyPool
            var CurrentDamageWindow:float=0.0
            for(Enemy->DamageData:EnemyPoolCopy):
                # Print("Apply Damage: {DamageData.AmountOfDamage}")
                Enemy.Damage(DamageData.AmountOfDamage) #inflict Damage
                if:
                    TargetData:=EnemyPool[Enemy]
                then:
                    set CurrentDamageWindow = TargetData.UpdateDamageWindow(-DamageDelay) #Check remaining time
                
                if(Enemy.GetHealth() <=0.0 or CurrentDamageWindow <= 0.0):  # if enemy is dead or Damage window is over, stop applying damage 
                    RemoveEnemyFromMap(Enemy)                            
                    StopVfx(Enemy)
#========================================================================================================================= 
    RemoveEnemyFromMap(Enemy:fort_character):void=
        var NewMap:[fort_character]booster_fire_bullets_damage_data=map{}
        # Concatenate Keys from ExampleMap into NewMap, excluding ElementToRemove
        for (Key -> Value : EnemyPool, Key <> Enemy):
            set NewMap = ConcatenateMaps(NewMap, map{Key => Value})
        set EnemyPool = NewMap
#========================================================================================================================= 
    AwaitForEnemyDamage()<suspends>:void=
        if(bDebugEnabled?){Logger.Print("Calling AwaitForEnemyDamage")}
        loop:
            Result:=GuardSpawner.DamagedEvent.Await()
            if(bDebugEnabled?){Logger.Print("Enemy Damage Detected")}
            var MaybeSourceFortCharacter:?fort_character=false
            if:
                Target:=Result.Target?
                TargetFortCharacter:=Target.GetFortCharacter[] 
            then:
                if:
                    Source:=Result.Source?
                    SourceFortCharacter:=Source.GetFortCharacter[] 
                then:
                    set MaybeSourceFortCharacter = option{SourceFortCharacter}
                OnEnemyDamagedEvent.Signal(TargetFortCharacter, MaybeSourceFortCharacter)