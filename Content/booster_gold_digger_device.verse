
using { /Fortnite.com/Devices }
using { /Verse.org/Simulation }
using { /UnrealEngine.com/Temporary/Diagnostics }
using { /Verse.org/Assets }

booster_upgrade_data_gold_digger := class<concrete>(booster_upgrade_data):
    BoosterDescriptionA<override>:string="Gold Digger "
    BoosterDescriptionB<override>:string=" Info"
    @editable
    ExtraGold : float = 2.0

#-----------------------------------------------------------------------------------------------------------------------

booster_default_upgrade_data_gold_digger := struct:

    DefaultValues: []booster_upgrade_data_gold_digger=array{
        Upgrade0:booster_upgrade_data_gold_digger = booster_upgrade_data_gold_digger{LifeSpawn := 60.0}
        Upgrade1:booster_upgrade_data_gold_digger = booster_upgrade_data_gold_digger{LifeSpawn := 60.0, ExtraGold := 6.0}
        Upgrade2:booster_upgrade_data_gold_digger = booster_upgrade_data_gold_digger{LifeSpawn := 60.0, ExtraGold := 8.0}
        Upgrade3:booster_upgrade_data_gold_digger = booster_upgrade_data_gold_digger{LifeSpawn := 60.0, ExtraGold := 10.0}
        Upgrade4:booster_upgrade_data_gold_digger = booster_upgrade_data_gold_digger{LifeSpawn := 60.0, ExtraGold := 12.0}
    }

#-----------------------------------------------------------------------------------------------------------------------

booster_gold_digger_device := class(booster_device):

    @editable
    EnemiesSpawner : guard_spawner_device = guard_spawner_device{}

    @editable
    var BoosterUpgradeData: []booster_upgrade_data_gold_digger = booster_default_upgrade_data_gold_digger{}.DefaultValues

    #Debug
    GetCurrencyInfo<localizes>(Player : player, Added : float) : message = "Extra Gold: {Added} added to: {Player}"

    BoosterID<override>:booster_id = booster_id.GoldDigger
    BoosterDisplayName<override>:string = "Gold Digger"
    BoosterIcon<override>:texture = BoosterSystem.Textures.Icons.Legacy.T_CoinRewards #Temporal icon

#-----------------------------------------------------------------------------------------------------------------------
    GetBoosterUpgradeData(BoosterUpgradeIndex:int)<decides><transacts>:booster_upgrade_data_gold_digger=
        BoosterUpgradeData[BoosterUpgradeIndex]
#==============================================================================================================================  
    GetBaseBoosterUpgradeData<override>(BoosterUpgradeIndex:int)<decides><transacts>:booster_upgrade_data=GetBoosterUpgradeData[BoosterUpgradeIndex]
#==============================================================================================================================
    GetMaxBoosterUpgradeIndex<override>():int=BoosterUpgradeData.Length
#==============================================================================================================================
    GetBoosterDescription<override>(BoosterUpgradeIndex:int):string= 
        if:
            BoosterData:=GetBoosterUpgradeData[BoosterUpgradeIndex]
        then:
            TruncatedString:=ToTruncatedString(BoosterData.ExtraGold,1)
            return BoosterData.BoosterDescriptionA + "{TruncatedString}" +BoosterData.BoosterDescriptionB
        return "Booster Description"
#==============================================================================================================================
    BoosterLoopBehavior<override>(Player:player, BoosterUpgradeIndex:int, BoosterLoopCompleteEvent:event(player), TerminateBoosterHandlerEvent:event())<suspends>:void=
        <#
        @Player: Player running this Booster
        @BoosterLoopCompleteEvent: if Loop is Complete Call the event to indicate booster was used
        @TerminateBoosterHandlerEvent.Await(): to finish this method, this is call from the main handler, if time is out or the execution limit was reached
        @TerminateLoopBehaviorEvent.Await() : To Finish this method and all sub rutines
        #>

        if(TargetBoosterData:=GetBoosterUpgradeData[BoosterUpgradeIndex]):

            TerminateLoopBehaviorEvent:event()=event(){}

            GoldAmount := TargetBoosterData.ExtraGold

            race:
              TerminateLoopBehaviorEvent.Await()
              sync:
                  loop:
                      TerminateBoosterHandlerEvent.Await()
                      Logger.Print("{BoosterDisplayName}: Loop Finished", ?Level:=log_level.Warning)
                      TerminateLoopBehaviorEvent.Signal()
                  loop:
                    EventInfo := EnemiesSpawner.EliminatedEvent.Await()

                    Source := EventInfo.Source                     
    
                      if (Agent := Source?, NPCKiller := player[Agent]):
                        if (NPCKiller = Player):

                            #Add extra gold to player
                            Print(GetCurrencyInfo(Player, GoldAmount))
    
                      BoosterLoopCompleteEvent.Signal(Player)
