using { /Fortnite.com/Devices }
using { /Fortnite.com/Characters }
using { /Verse.org/Simulation }
using { /Verse.org/Assets }
using { /UnrealEngine.com/Temporary/Diagnostics }
using { /UnrealEngine.com/Temporary/SpatialMath }
using { /Verse.org/Random }

#Booster Set Up
#-----------------------------------------------------------------------------------------------------------------

booster_upgrade_data_healthy_drops := class<concrete>(booster_upgrade_data):
    LifeSpawn<override>:float=-1.0
    BoosterDescriptionA<override>:string="% chance for \nenemies dropping \na healing item"
    BoosterDescriptionB<override>:string=""
    @editable
    DropChance:int = 5

booster_default_upgrade_data_healthy_drops:=struct:

    DefaultValues:[]booster_upgrade_data_healthy_drops = array{
        Upgrade0:booster_upgrade_data_healthy_drops = booster_upgrade_data_healthy_drops{}
        Upgrade1:booster_upgrade_data_healthy_drops = booster_upgrade_data_healthy_drops{DropChance := 10}
        Upgrade2:booster_upgrade_data_healthy_drops = booster_upgrade_data_healthy_drops{DropChance := 20}
        Upgrade3:booster_upgrade_data_healthy_drops = booster_upgrade_data_healthy_drops{DropChance := 30}
    }

#Booster Data
#-----------------------------------------------------------------------------------------------------------------

booster_healthy_drops_device := class(booster_device):

    @editable
    var BoosterUpgradeData:[]booster_upgrade_data_healthy_drops = booster_default_upgrade_data_healthy_drops{}.DefaultValues

    @editable
    ItemSpawner:item_spawner_device = item_spawner_device{}
    @editable
    GuardSpawner:guard_spawner_device = guard_spawner_device{}

    BoosterID<override>:booster_id = booster_id.HealthyDrops
    BoosterDisplayName<override>:string = "Healthy Drops"
    BoosterIcon<override>:texture=BoosterSystem.Textures.Icons.Legacy.T_HealthyDrops

#Booster Functions
#-----------------------------------------------------------------------------------------------------------------

    GetBoosterUpgradeData(BoosterUpgradeIndex:int)<decides><transacts>:booster_upgrade_data_healthy_drops =
        BoosterUpgradeData[BoosterUpgradeIndex]
#==============================================================================================================================      
    GetBaseBoosterUpgradeData<override>(BoosterUpgradeIndex:int)<decides><transacts>:booster_upgrade_data=GetBoosterUpgradeData[BoosterUpgradeIndex]
#==============================================================================================================================
    GetMaxBoosterUpgradeIndex<override>():int=BoosterUpgradeData.Length
#==============================================================================================================================
    GetBoosterDescription<override>(BoosterUpgradeIndex:int):string= 
        if:
            BoosterData:=GetBoosterUpgradeData[BoosterUpgradeIndex]
        then:
            TruncatedString:=ToTruncatedString(BoosterData.DropChance * 1.0,1)
            return "{TruncatedString}" + BoosterData.BoosterDescriptionA 
        return "Booster Description"
#==============================================================================================================================
    BoosterLoopBehavior<override>(Player:player, BoosterUpgradeIndex:int, BoosterLoopCompleteEvent:event(player), TerminateBoosterHandlerEvent:event())<suspends>:void =

        if(TargetBoosterData:=GetBoosterUpgradeData[BoosterUpgradeIndex]):
        
            TerminateLoopBehaviorEvent:event()=event(){}

            DropChance := TargetBoosterData.DropChance

            if(FortCharacter:=Player.GetFortCharacter[]):
                        
                race:
                    TerminateLoopBehaviorEvent.Await()
                    sync:
                        loop:
                            TerminateBoosterHandlerEvent.Await()
                            Logger.Print("{BoosterDisplayName}: Loop Finished", ?Level:=log_level.Warning)
                            TerminateLoopBehaviorEvent.Signal()
                        loop:
                            Sleep(1.0)
                            Logger.Print("{BoosterDisplayName}: Running, Upgrade Index: {BoosterUpgradeIndex}")
                            BoosterLoopCompleteEvent.Signal(Player)
                        loop:
                            #TODO Replace with Enemies Manager to detect when an NPC has died and get the agent
                            Result := GuardSpawner.EliminatedEvent.Await()
                            QAgent := Result.Target

                            if:
                                QAgent <> false
                                Agent: agent = QAgent?
                                EliminatedCharacter: fort_character = Agent.GetFortCharacter[]
                                EliminatedTransform := EliminatedCharacter.GetTransform()
                                RandomGeneratedInt := GetRandomInt(0, 100)
                            then:
                                if(RandomGeneratedInt <= DropChance):
                                    if(ItemSpawner.TeleportTo[EliminatedTransform.Translation + vector3{Z:=0.002}, EliminatedTransform.Rotation]):
                                        ItemSpawner.SpawnItem()
