using { /Fortnite.com/Devices }
using { /Fortnite.com/Characters }
using { /Verse.org/Simulation }
using { /Verse.org/Assets }
using { /UnrealEngine.com/Temporary/Diagnostics }

booster_upgrade_data_health_regen:=class<concrete>(booster_upgrade_data):
    LifeSpawn<override>:float=120.0
    BoosterDescriptionA<override>:string=" Health Regen"
    BoosterDescriptionB<override>:string=""
    @editable
    AmountOfHealth:float=0.5
    @editable
    MaxAmountOfHealth:float=0.5
    @editable
    RegenDelay:float=3.0
    @editable
    MaxLifeSpawn:float=360.0

booster_default_upgrade_data_health_regen:=struct:

    DefaultValues:[]booster_upgrade_data_health_regen=array{
        Upgrade0:booster_upgrade_data_health_regen=booster_upgrade_data_health_regen{}
        Upgrade1:booster_upgrade_data_health_regen=booster_upgrade_data_health_regen{AmountOfHealth:=1.0,MaxAmountOfHealth:=1.0}
        Upgrade2:booster_upgrade_data_health_regen=booster_upgrade_data_health_regen{AmountOfHealth:=1.5,MaxAmountOfHealth:=1.5}
        Upgrade3:booster_upgrade_data_health_regen=booster_upgrade_data_health_regen{AmountOfHealth:=2.0,MaxAmountOfHealth:=2.0}
    }
#==============================================================================================================================
booster_health_regen_device := class(booster_device):

    @editable
    var BoosterUpgradeData:[]booster_upgrade_data_health_regen=booster_default_upgrade_data_health_regen{}.DefaultValues

    BoosterID<override>:booster_id=booster_id.HealthRegen
    BoosterDisplayName<override>:string="Health Regen"
    BoosterIcon<override>:texture=BoosterSystem.Textures.Icons.Legacy.T_IncreaseHealthRegen
    bIsStackable<override>:logic=true
#==============================================================================================================================
    GetBoosterUpgradeData(BoosterUpgradeIndex:int)<decides><transacts>:booster_upgrade_data_health_regen=
        BoosterUpgradeData[BoosterUpgradeIndex]
#==============================================================================================================================      
    GetBaseBoosterUpgradeData<override>(BoosterUpgradeIndex:int)<decides><transacts>:booster_upgrade_data=GetBoosterUpgradeData[BoosterUpgradeIndex]
#==============================================================================================================================
    GetMaxBoosterUpgradeIndex<override>():int=BoosterUpgradeData.Length
#==============================================================================================================================
    GetBoosterDescription<override>(BoosterUpgradeIndex:int):string= 
        if:
            BoosterData:=GetBoosterUpgradeData[BoosterUpgradeIndex]
        then:
            TruncatedString:=ToTruncatedString(BoosterData.AmountOfHealth,1)
            return "{TruncatedString}" + BoosterData.BoosterDescriptionA 
        return "Booster Description"
#==============================================================================================================================
    BoosterLoopBehavior<override>(Player:player, BoosterUpgradeIndex:int, BoosterLoopCompleteEvent:event(player), TerminateBoosterHandlerEvent:event())<suspends>:void=
        if(TargetBoosterData:=GetBoosterUpgradeData[BoosterUpgradeIndex]):

            TerminateLoopBehaviorEvent:event()=event(){}

            var AmountOfHealth:float=TargetBoosterData.AmountOfHealth
            MaxAmountOfHealth:=TargetBoosterData.MaxAmountOfHealth
            RegenDelay:=TargetBoosterData.RegenDelay

            if(FortCharacter:=Player.GetFortCharacter[]):
                
                race:
                    TerminateLoopBehaviorEvent.Await()
                    sync:
                        loop:
                            TerminateBoosterHandlerEvent.Await()
                            Logger.Print("{BoosterDisplayName}: Loop Finished", ?Level:=log_level.Warning)
                            TerminateLoopBehaviorEvent.Signal()
                        loop:
                            EventData:=OnBoosterNewInstanceStartedEvent.Await()
                            EventPlayer:=EventData(0)
                            EventBoosterID:=EventData(1)  
                            EventBoosterUpgradeIndex:=EventData(2) 
                            if(EventPlayer = Player):
                                if(bIsStackable?):
                                    if(NewTargetBoosterData:=GetBoosterUpgradeData[BoosterUpgradeIndex]):
                                        NewAmountOfHealth:=AmountOfHealth + NewTargetBoosterData.AmountOfHealth
                                        set AmountOfHealth = Clamp(NewAmountOfHealth, 0.0, MaxAmountOfHealth)
                                        Logger.Print("{BoosterDisplayName}: Health Regen Stacked: {AmountOfHealth}", ?Level:=log_level.Warning)
                        loop:
                            Sleep(RegenDelay)
                            if(FortCharacter.GetHealth() < FortCharacter.GetMaxHealth()):
                                FortCharacter.Heal(AmountOfHealth)
                            BoosterLoopCompleteEvent.Signal(Player)

    OnBoosterNewInstanceStarted<override>(Player:player, NewBoosterUpgradeIndex:int, CurrentBoosterRunData:booster_run_data):void=
        if(NewTargetBoosterData:=GetBoosterUpgradeData[NewBoosterUpgradeIndex]):
            if(bIsStackable?):
                LifeSpawn:= CurrentBoosterRunData.LifeSpawnLeft + NewTargetBoosterData.LifeSpawn
                Limit:=NewTargetBoosterData.MaxLifeSpawn
                NewLifeSpawn:=Clamp(LifeSpawn, 0.0, Limit)
                Print("Updating Run Data, New Life Spawn: {NewLifeSpawn}")
                CurrentBoosterRunData.UpdateLifeSpawn(NewLifeSpawn)
                OnBoosterRunDataUpdatedEvent.Signal(CurrentBoosterRunData)