
using { /Fortnite.com/Devices }
using { /Verse.org/Simulation }
using { /Verse.org/Simulation/Tags }
using { /UnrealEngine.com/Temporary/Diagnostics }
using { /UnrealEngine.com/Temporary/UI }
using { /UnrealEngine.com/Temporary/SpatialMath }

#=========================================================================================================================
booster_manager_device := class(creative_device):

    Logger:log=log{Channel:=log_booster_manager, DefaultLevel:=log_level.Debug}

    var AvailableBoosters:[]booster_device=array{}

    @editable
    bDebugEnabled:logic=false
    #Set the desired tag here
    @editable
    SearchingTag:tag = booster_device_tag{}
    @editable
    BoosterIconSize:vector2=vector2{X:=35.0,Y:=35.0}
    @editable
    BoosterIconLateralPadding:float=5.0
    @editable
    BoosterStackBoxSize:vector2=vector2{X:=650.0,Y:=70.0}
    @editable
    BoosterStackBoxOffSet:vector2=vector2{X:=35.0,Y:=15.0}

    var BoostersInstances:[player]booster_instances=map{}
#=========================================================================================================================
    OnBegin<override>()<suspends>:void=
        if(bDebugEnabled?):
            Logger.Print("Starting...")

        set AvailableBoosters = GetAllBoosterInWorld()
        if(bDebugEnabled?):
            Logger.Print("Boosters Found: {AvailableBoosters.Length}")

        PlaySpace:=GetPlayspace()
        PlaySpace.PlayerRemovedEvent().Subscribe(OnPlayerRemovedEvent)
#=========================================================================================================================
    GetAllBoosterInWorld():[]booster_device=
        var TmpAvailableBoosters:[]booster_device=array{}
        #Get all tagged objects using @SearchingTag
        CreativeObjects:=FindCreativeObjectsWithTag(SearchingTag)
        #Check if object class Match @booster_device and is Enabled
        for(CreativeObject:CreativeObjects):
            if:
                ValidBoosterDevice:=booster_device[CreativeObject]
                ValidBoosterDevice.bEnabled?
            then:
                set TmpAvailableBoosters+=array{ValidBoosterDevice}

        return TmpAvailableBoosters
#=========================================================================================================================
    GetBoosterDeviceByID(BoosterID:booster_id):?booster_device=
        for(TargetBooster:AvailableBoosters):
            if(TargetBooster.BoosterID = BoosterID):
                return option{TargetBooster}
        return false
#=========================================================================================================================
    CreateBoostersWidget(Player:player, ?TargetCanvas:?canvas=false):tuple(canvas, stack_box)=      
 
        #if TargetCanvas = false, create a new canvas

        #Canvas to store the booster's icon
        var BoosterCanvas:canvas=canvas{}

        if(ValidCanvas:=TargetCanvas?):
            set BoosterCanvas = ValidCanvas

       #Main Overlay
       #Define Main Widget
       MainOverlay:=overlay{}
       BoostersStackBox:=stack_box{Orientation:=orientation.Horizontal}

       #Add Memebers
       Slot:=overlay_slot{
        Widget:=BoostersStackBox 
        HorizontalAlignment:=horizontal_alignment.Fill
        VerticalAlignment:=vertical_alignment.Fill
        Padding:=margin{Left:=5.0,Top:=5.0,Right:=5.0,Bottom:=5.0}
        }

       MainOverlay.AddWidget(Slot)

       BoosterCanvas.AddWidget(CanvasSlot:=canvas_slot{
        Widget:=MainOverlay
        Anchors:=anchors{Minimum:=vector2{X:=0.0,Y:=1.0}, Maximum:=vector2{X:=0.0,Y:=1.0}}
        Offsets:=margin{Left:=BoosterStackBoxOffSet.X, Top:=BoosterStackBoxOffSet.Y, Right:=BoosterStackBoxSize.X, Bottom:=BoosterStackBoxSize.Y}
        Alignment:=vector2{X:=0.0, Y:=1.0}
       })

        if(PlayerUI:=GetPlayerUI[Player]):
            PlayerUI.AddWidget(BoosterCanvas)

        return(BoosterCanvas, BoostersStackBox)
#=========================================================================================================================
    GiveBoosterToPlayer(Player:player, BoosterID:booster_id, BoosterUpgradeIndex:int, ?TargetCanvas:?canvas=false):void=
        if(AvailableBoosters.Length > 0):
            if(bDebugEnabled?):
                Logger.Print("Finding Matching Booster by ID")
            MaybeBooster:=GetBoosterDeviceByID(BoosterID)
            if(TargetBooster:=MaybeBooster?):
                #Register the given booster
                AddBoosterInstance(Player, BoosterID, BoosterUpgradeIndex, ?TargetCanvas:=TargetCanvas)
                #Execute the target booster device
                InitializeBoosterForPlayer(Player, TargetBooster, BoosterUpgradeIndex)
        else:
            if(bDebugEnabled?):
                Logger.Print("There are not Boosters set",?Level:=log_level.Error)
#=========================================================================================================================
    InitializeBoosterForPlayer(Player:player, BoosterDevice:booster_device, BoosterUpgradeIndex:int, ?TargetCanvas:?canvas=false):void=
        bIsBoosterInstanceRunning:=IsBoosterInstanceRunning(Player, BoosterDevice.BoosterID)
        if(bDebugEnabled?):
            if(bIsBoosterInstanceRunning?):
                Print("{BoosterDevice.BoosterDisplayName} Is Running")
            else:
                Print("{BoosterDevice.BoosterDisplayName} Not Running")

        if(not bIsBoosterInstanceRunning? or bIsBoosterInstanceRunning? and BoosterDevice.bIsStackable?):
            BoosterDevice.AddBoosterToPlayer(Player, BoosterUpgradeIndex, ?IsAdditionalInstance:=bIsBoosterInstanceRunning)
            spawn{OnBoosterFinished(Player, BoosterDevice)} #Corutine to wait until booster is finished
            if(TargetInstanceData:=BoostersInstances[Player]):
                Print("{BoosterDevice.BoosterDisplayName} Is Running Set")
                TargetInstanceData.SetBoosterIsRunning(BoosterDevice.BoosterID, true)

#=========================================================================================================================
    #Check if there is an intance of this booster if not, add it
    AddBoosterInstance(Player:player, BoosterID:booster_id, BoosterUpgradeIndex:int,?TargetCanvas:?canvas=false):void=

        CurrentBoosterInstances:=GetNumberOfBoosterIntances(Player, BoosterID)
        if(bDebugEnabled?){Logger.Print("Current Booster Instances for this Player: {CurrentBoosterInstances}")}

        if(not BoostersInstances[Player]):
            if(set BoostersInstances[Player] = booster_instances{}):
                if(bDebugEnabled?):
                    Logger.Print("Creating New Pool for this booster")

        if(TargetInstanceData:=BoostersInstances[Player]):
            #Register this booster Instance
            TargetInstanceData.AddInstance(BoosterID, BoosterUpgradeIndex)   
            #UI
            bIsCanvasCreated:logic=TargetInstanceData.bIsCanvasCreated

            if(not bIsCanvasCreated?):
                if(bDebugEnabled?){Logger.Print("Creating New Canvas for this Player")}
                    
                PlayerWidgets:=CreateBoostersWidget(Player, ?TargetCanvas:=TargetCanvas)
                NewCanvas:=PlayerWidgets(0)
                NewStackBox:=PlayerWidgets(1)
                TargetInstanceData.SetPlayerWidgetData(NewCanvas, NewStackBox)

            if(bDebugEnabled?){Logger.Print("Getting Canvas for this Player")}

            MaybeBooster:=GetBoosterDeviceByID(BoosterID)
            if(TargetBooster:=MaybeBooster?):
    
                CurrentCanvas:=TargetInstanceData.PlayerCanvas
                CurrentStackBox:=TargetInstanceData.PlayerStackBox

                #Check Instances if = 0, Add new Widget, if > 0, increase the Amount variable

                if(CurrentBoosterInstances = 0):
                    if(bDebugEnabled?){Logger.Print("Adding New Booster Widget")}

                    CurrentStackBox.AddWidget(Slot:=stack_box_slot{
                        Widget:=TargetBooster.GenerateBoosterWidget(Player, BoosterIconSize, BoosterUpgradeIndex, CurrentCanvas)
                        Padding:=margin{Left:=BoosterIconLateralPadding,Top:=0.0,Right:=BoosterIconLateralPadding,Bottom:=5.0}
                        HorizontalAlignment:=horizontal_alignment.Center
                        VerticalAlignment:=vertical_alignment.Top
                    })
                else:
                    #Update On Widget
                    if(bDebugEnabled?){Logger.Print("Updating Booster Widget")}
                    TargetBooster.OnBoosterNewInstanceAddedEvent.Signal(Player, BoosterID, CurrentBoosterInstances+1, BoosterUpgradeIndex)

            #Debug
            if(bDebugEnabled?):
                NumberOfIntances:=GetNumberOfBoosterIntances(Player,BoosterID)
                Logger.Print("Adding booster Intance, Total: {NumberOfIntances}")    
               
#=========================================================================================================================
#Return Remaining Intances
    RemoveBoosterInstance(Player:player, BoosterID:booster_id):int=
        if(TargetInstanceData:=BoostersInstances[Player]):
            Remain:=TargetInstanceData.RemoveInstance(BoosterID)
            if(Remain=0):
                if(bDebugEnabled?):
                    Logger.Print("Not more instances of this Booster")
            else:
                if(bDebugEnabled?):
                    Logger.Print("Launch Next Instance")
                return Remain
        return 0
#=========================================================================================================================
    RemoveAllBoosterInstances<protected>(Player:player, BoosterID:booster_id):void=
        if(TargetInstanceData:=BoostersInstances[Player]):
            TargetInstanceData.RemoveAllInstances(BoosterID)
#=========================================================================================================================
    OnBoosterFinished(Player:player, BoosterInstance:booster_device)<suspends>:void=
        loop:
            FinishBoosterData:=BoosterInstance.OnBoosterFinishedEvent.Await()
            FinishPlayer:=FinishBoosterData(0)
            FinishBoosterId:=FinishBoosterData(1)
            FinishWasPlayerEliminated:=FinishBoosterData(2)
            if(Player = FinishPlayer):
                var Remain:int=0
                if(FinishWasPlayerEliminated?):
                    #Remove all instances
                    RemoveAllBoosterInstances(Player, FinishBoosterId)
                    if(bDebugEnabled?){Logger.Print("Player eliminated, removing all booster intances")}
                else:
                    set Remain = RemoveBoosterInstance(Player, FinishBoosterId)

                BoosterInstance.OnBoosterInstanceRemovedEvent.Signal(Player,FinishBoosterId, Remain)
                #Change booster running state while new instance is started
                if(TargetInstanceData:=BoostersInstances[Player]):
                    TargetInstanceData.SetBoosterIsRunning(FinishBoosterId, false)

                #If there are more instances of this booster Launch them
                if:
                    Remain>0
                    AllBoostersInstances:=BoostersInstances[Player] #All boosters Instances for this Player
                    BoosterPool:=AllBoostersInstances.BoostersPool[FinishBoosterId] #All Instances of this Booster
                    BoosterInstanceData:=BoosterPool.Data[0] #Get the first Booster Instance in the array
                then:            
                    if(not BoosterInstance.bIsStackable?):
                        InitializeBoosterForPlayer(Player, BoosterInstance, BoosterInstanceData.BoosterUpgradeIndex)
                    break
                else:
                    break
#=========================================================================================================================
    GetNumberOfBoosterIntances(Player:player, BoosterID:booster_id):int=
        if(Target_InstanceData:=BoostersInstances[Player]):
            return Target_InstanceData.GetNumberOfIntances(BoosterID)
        return 0
#=========================================================================================================================
    IsBoosterInstanceRunning(Player:player, BoosterID:booster_id):logic=
        if:
            Target_InstanceData:=BoostersInstances[Player]     
        then:          
            return Target_InstanceData.GetBoosterIsRunning(BoosterID)
        return false
#=========================================================================================================================   
    RemoveAllBoostersFromPlayer(Player:player):void=
        if(Target_InstanceData:=BoostersInstances[Player]):
            Pool:=Target_InstanceData.BoostersPool
            for(key->value : Pool): #booster_id -> booster_instance_attributes_data
                RemoveAllBoosterInstances(Player, key)
                MaybeBooster:= GetBoosterDeviceByID(key)
                if(ValidBooster:=MaybeBooster?):
                    ValidBooster.OnBoosterFinished(Player)
#========================================================================================================================= 
    OnPlayerRemovedEvent(Player:player):void=
        RemoveAllBoostersFromPlayer(Player)
        RemovePlayerFromMap(Player)
#========================================================================================================================= 
    RemovePlayerFromMap(Player:player):void=
        var NewMap:[player]booster_instances=map{}
        # Concatenate Keys from ExampleMap into NewMap, excluding ElementToRemove
        for (Key -> Value : BoostersInstances, Key <> Player):
            set NewMap = ConcatenateMaps(NewMap, map{Key => Value})
        set BoostersInstances = NewMap
        


