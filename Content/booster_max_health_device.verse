using { /Fortnite.com/Devices }
using { /Fortnite.com/Characters }
using { /Verse.org/Simulation }
using { /Verse.org/Assets }
using { /UnrealEngine.com/Temporary/Diagnostics }

booster_upgrade_data_max_health:=class<concrete>(booster_upgrade_data):
    LifeSpawn<override>:float=-1.0
    BoosterDescriptionA<override>:string="+"
    BoosterDescriptionB<override>:string=" Max Health \nuntil death"
    @editable
    AmountOfHealth:float=5.0

booster_default_upgrade_data_max_health:=struct:

    DefaultValues:[]booster_upgrade_data_max_health=array{
        Upgrade0:booster_upgrade_data_max_health=booster_upgrade_data_max_health{}
        Upgrade1:booster_upgrade_data_max_health=booster_upgrade_data_max_health{AmountOfHealth:=15.0}
        Upgrade2:booster_upgrade_data_max_health=booster_upgrade_data_max_health{AmountOfHealth:=25.0}
        Upgrade3:booster_upgrade_data_max_health=booster_upgrade_data_max_health{AmountOfHealth:=35.0}
    }


booster_max_health_device := class(booster_device):

    @editable
    var BoosterUpgradeData:[]booster_upgrade_data_max_health=booster_default_upgrade_data_max_health{}.DefaultValues

    BoosterID<override>:booster_id=booster_id.MaxHealth
    BoosterDisplayName<override>:string="Max Health"
    BoosterIcon<override>:texture=BoosterSystem.Textures.Icons.Legacy.T_IncreaseMaxHealth
    bIsStackable<override>:logic=true
#==============================================================================================================================
    GetBoosterUpgradeData(BoosterUpgradeIndex:int)<decides><transacts>:booster_upgrade_data_max_health=
        BoosterUpgradeData[BoosterUpgradeIndex]
#==============================================================================================================================       
    GetBaseBoosterUpgradeData<override>(BoosterUpgradeIndex:int)<decides><transacts>:booster_upgrade_data=GetBoosterUpgradeData[BoosterUpgradeIndex]
#==============================================================================================================================
    GetMaxBoosterUpgradeIndex<override>():int=BoosterUpgradeData.Length
#==============================================================================================================================
    GetBoosterDescription<override>(BoosterUpgradeIndex:int):string= 
        if:
            BoosterData:=GetBoosterUpgradeData[BoosterUpgradeIndex]
        then:
            TruncatedString:=ToTruncatedString(BoosterData.AmountOfHealth,1)
            return BoosterData.BoosterDescriptionA + "{TruncatedString}" + BoosterData.BoosterDescriptionB
        return "Booster Description"
#==============================================================================================================================
    BoosterLoopBehavior<override>(Player:player, BoosterUpgradeIndex:int, BoosterLoopCompleteEvent:event(player), TerminateBoosterHandlerEvent:event())<suspends>:void=
        if(TargetBoosterData:=GetBoosterUpgradeData[BoosterUpgradeIndex]):

            TerminateLoopBehaviorEvent:event()=event(){}
            var AmountOfHealth:float=TargetBoosterData.AmountOfHealth
            UpdatePlayerMaxHealth(Player, AmountOfHealth)
            
            race:
                TerminateLoopBehaviorEvent.Await()
                
                sync:

                    loop:
                        TerminateBoosterHandlerEvent.Await()
                        Logger.Print("{BoosterDisplayName}: Loop Finished", ?Level:=log_level.Warning)
                        #Restore Default Max Health
                        UpdatePlayerMaxHealth(Player, -AmountOfHealth)
                        TerminateLoopBehaviorEvent.Signal()
                    loop:
                        EventData:=OnBoosterNewInstanceStartedEvent.Await()
                        EventPlayer:=EventData(0)
                        EventBoosterID:=EventData(1)  
                        EventBoosterUpgradeIndex:=EventData(2) 
                        if(EventPlayer = Player):
                            if(bIsStackable?):
                                if(NewTargetBoosterData:=GetBoosterUpgradeData[BoosterUpgradeIndex]):
                                    #Remove Prev Update, in case other device increases this player max health, we ensure only iterate over the amount we had increased
                                    UpdatePlayerMaxHealth(Player, -AmountOfHealth)
                                    set AmountOfHealth += NewTargetBoosterData.AmountOfHealth
                                    #Set New Max Health
                                    UpdatePlayerMaxHealth(Player, AmountOfHealth)
                                    Logger.Print("{BoosterDisplayName}: Max Health Stacked: {AmountOfHealth}", ?Level:=log_level.Warning)
                    loop:
                        Sleep(0.0)
                        BoosterLoopCompleteEvent.Signal(Player)
                        break
#==============================================================================================================================
    UpdatePlayerMaxHealth(Player:player, AmountToIncrease:float):void=
        if(FortCharacter:=Player.GetFortCharacter[]):
            CurrentMaxHealth:=FortCharacter.GetMaxHealth()
            NewMaxHealth:= CurrentMaxHealth+AmountToIncrease
            FortCharacter.SetMaxHealth(NewMaxHealth)
