using { /Fortnite.com/Devices }
using { /Fortnite.com/Characters }
using { /Verse.org/Simulation }
using { /Verse.org/Assets }
using { /UnrealEngine.com/Temporary/Diagnostics }

booster_upgrade_data_max_shield:=class<concrete>(booster_upgrade_data):
    LifeSpawn<override>:float=-1.0
    BoosterDescriptionA<override>:string="+"
    BoosterDescriptionB<override>:string=" Max Shield \nuntil death"
    @editable
    AmountOfShield:float=5.0

booster_default_upgrade_data_max_shield:=struct:

    DefaultValues:[]booster_upgrade_data_max_shield=array{
        Upgrade0:booster_upgrade_data_max_shield=booster_upgrade_data_max_shield{}
        Upgrade1:booster_upgrade_data_max_shield=booster_upgrade_data_max_shield{AmountOfShield:=15.0}
        Upgrade2:booster_upgrade_data_max_shield=booster_upgrade_data_max_shield{AmountOfShield:=25.0}
        Upgrade3:booster_upgrade_data_max_shield=booster_upgrade_data_max_shield{AmountOfShield:=35.0}
    }

#==============================================================================================================================
booster_max_shield_device := class(booster_device):

    @editable
    var BoosterUpgradeData:[]booster_upgrade_data_max_shield=booster_default_upgrade_data_max_shield{}.DefaultValues

    BoosterID<override>:booster_id=booster_id.MaxShield
    BoosterDisplayName<override>:string="Max Shield"
    BoosterIcon<override>:texture=BoosterSystem.Textures.Icons.Legacy.T_IncreaseMaxShield  
    bIsStackable<override>:logic=true
#==============================================================================================================================
    GetBoosterUpgradeData(BoosterUpgradeIndex:int)<decides><transacts>:booster_upgrade_data_max_shield=
        BoosterUpgradeData[BoosterUpgradeIndex]
 #==============================================================================================================================       
    GetBaseBoosterUpgradeData<override>(BoosterUpgradeIndex:int)<decides><transacts>:booster_upgrade_data=GetBoosterUpgradeData[BoosterUpgradeIndex]
#==============================================================================================================================
    GetMaxBoosterUpgradeIndex<override>():int=BoosterUpgradeData.Length
#==============================================================================================================================
    GetBoosterDescription<override>(BoosterUpgradeIndex:int):string= 
        if:
            BoosterData:=GetBoosterUpgradeData[BoosterUpgradeIndex]
        then:
            TruncatedString:=ToTruncatedString(BoosterData.AmountOfShield,1)
            return BoosterData.BoosterDescriptionA + "{TruncatedString}" + BoosterData.BoosterDescriptionB
        return "Booster Description"
#==============================================================================================================================
    BoosterLoopBehavior<override>(Player:player, BoosterUpgradeIndex:int, BoosterLoopCompleteEvent:event(player), TerminateBoosterHandlerEvent:event())<suspends>:void=
        if(TargetBoosterData:=GetBoosterUpgradeData[BoosterUpgradeIndex]):

            TerminateLoopBehaviorEvent:event()=event(){}

            var AmountOfShield:float=TargetBoosterData.AmountOfShield
            UpdatePlayerMaxShield(Player, AmountOfShield)
            
            race:
                TerminateLoopBehaviorEvent.Await()
                
                sync:

                    loop:
                        TerminateBoosterHandlerEvent.Await()
                        Logger.Print("{BoosterDisplayName}: Loop Finished", ?Level:=log_level.Warning)
                        #Restore Default Max Shield
                        UpdatePlayerMaxShield(Player, -AmountOfShield)
                        TerminateLoopBehaviorEvent.Signal()
                    loop:
                        EventData:=OnBoosterNewInstanceStartedEvent.Await()
                        EventPlayer:=EventData(0)
                        EventBoosterID:=EventData(1)  
                        EventBoosterUpgradeIndex:=EventData(2) 
                        if(EventPlayer = Player):
                            if(bIsStackable?):
                                if(NewTargetBoosterData:=GetBoosterUpgradeData[BoosterUpgradeIndex]):
                                    #Remove Prev Update, in case other device increases this player max shield, we ensure only iterate over the amount we had increased
                                    UpdatePlayerMaxShield(Player, -AmountOfShield)
                                    set AmountOfShield += NewTargetBoosterData.AmountOfShield
                                    #Set New Max Shield
                                    UpdatePlayerMaxShield(Player, AmountOfShield)
                                    Logger.Print("{BoosterDisplayName}: Max Shield Stacked: {AmountOfShield}", ?Level:=log_level.Warning)
                    loop:
                        Sleep(0.0)
                        BoosterLoopCompleteEvent.Signal(Player)
                        break
#==============================================================================================================================
    UpdatePlayerMaxShield(Player:player, AmountToIncrease:float):void=
        if(FortCharacter:=Player.GetFortCharacter[]):
            CurrentMaxShield:=FortCharacter.GetMaxShield()
            NewMaxshield:= CurrentMaxShield+AmountToIncrease
            FortCharacter.SetMaxShield(NewMaxshield)