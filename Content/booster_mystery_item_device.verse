
using { /Fortnite.com/Devices }
using { /Verse.org/Simulation }
using { /UnrealEngine.com/Temporary/Diagnostics }
using { /Verse.org/Assets }
using { /Verse.org/Random}

booster_upgrade_data_mystery_item := class<concrete>(booster_upgrade_data):

    ExecutionsLimit<override>: int = 1
    BoosterDescriptionA<override>:string="Feeling lucky?"

    @editable
    L1Chance : int = 25

    @editable
    L2Chance : int = 70

    @editable
    L3Chance : int = 5

    @editable
    L4Chance : int = 0

    @editable
    L5Chance : int = 0

    @editable
    L6Chance : int = 0

#-----------------------------------------------------------------------------------------------------------------------

booster_default_upgrade_data_mystery_item := struct:

    DefaultValues: []booster_upgrade_data_mystery_item=array{
        Upgrade0:booster_upgrade_data_mystery_item = booster_upgrade_data_mystery_item{}
        Upgrade1:booster_upgrade_data_mystery_item = booster_upgrade_data_mystery_item{L1Chance := 10, L2Chance := 50, L3Chance := 40}
        Upgrade2:booster_upgrade_data_mystery_item = booster_upgrade_data_mystery_item{L1Chance := 0, L2Chance := 25, L3Chance := 45, L4Chance := 0, L5Chance := 25, L6Chance := 5}
    }

#-----------------------------------------------------------------------------------------------------------------------

booster_mystery_item_device := class(booster_device):

    @editable
    L1ItemGranter : item_granter_device = item_granter_device{}

    @editable
    L2ItemGranter : item_granter_device = item_granter_device{}

    @editable
    L3ItemGranter : item_granter_device = item_granter_device{}

    @editable
    L4ItemGranter : item_granter_device = item_granter_device{}

    @editable
    L5ItemGranter : item_granter_device = item_granter_device{}

    @editable
    L6ItemGranter : item_granter_device = item_granter_device{}

    @editable
    var BoosterUpgradeData: []booster_upgrade_data_mystery_item = booster_default_upgrade_data_mystery_item{}.DefaultValues

    BoosterID<override>:booster_id = booster_id.MysteryItem
    BoosterDisplayName<override>:string = "Mystery Item"
    BoosterIcon<override>:texture = BoosterSystem.Textures.Icons.Legacy.T_CoinRewards #Temporal icon

#-----------------------------------------------------------------------------------------------------------------------

    GetBoosterUpgradeData(BoosterUpgradeIndex:int)<decides><transacts>:booster_upgrade_data_mystery_item=
        BoosterUpgradeData[BoosterUpgradeIndex]
#==============================================================================================================================
    GetBaseBoosterUpgradeData<override>(BoosterUpgradeIndex:int)<decides><transacts>:booster_upgrade_data=GetBoosterUpgradeData[BoosterUpgradeIndex]
#==============================================================================================================================
    GetMaxBoosterUpgradeIndex<override>():int=BoosterUpgradeData.Length
    #==============================================================================================================================
    GetBoosterDescription<override>(BoosterUpgradeIndex:int):string=
        if:
            BoosterData:=GetBoosterUpgradeData[BoosterUpgradeIndex]
        then:
            return BoosterData.BoosterDescriptionA
        return "Booster Description"
#==============================================================================================================================
    GetRandomLevel(L1Chance : int, L2Chance : int, L3Chance : int, L4Chance : int, L5Chance : int, L6Chance : int): int=

        TotalChance := L1Chance + L2Chance + L3Chance + L4Chance + L5Chance + L6Chance
        RandomNumber := GetRandomInt(0, TotalChance)

        if (RandomNumber < L1Chance):
            return 1
        if (RandomNumber < L1Chance + L2Chance):
            return 2
        if (RandomNumber < L1Chance + L2Chance + L3Chance):
            return 3
        if (RandomNumber < L1Chance + L2Chance + L3Chance + L4Chance):
            return 4
        if (RandomNumber < L1Chance + L2Chance + L3Chance + L4Chance + L5Chance):
            return 5

        return 6
#==============================================================================================================================
    GetItemGranter (Level : int): item_granter_device=
        if (Level = 1):
            return L1ItemGranter
        if (Level = 2):
            return L2ItemGranter
        if (Level = 3):
            return L3ItemGranter
        if (Level = 4):
            return L4ItemGranter
        if (Level = 5):
            return L5ItemGranter
        if (Level = 6):
            return L6ItemGranter

        return L6ItemGranter
#==============================================================================================================================
    BoosterLoopBehavior<override>(Player:player, BoosterUpgradeIndex:int, BoosterLoopCompleteEvent:event(player), TerminateBoosterHandlerEvent:event())<suspends>:void =

        if(TargetBoosterData:=GetBoosterUpgradeData[BoosterUpgradeIndex]):
        
            TerminateLoopBehaviorEvent:event()=event(){}

            L1Chance := TargetBoosterData.L1Chance
            L2Chance := TargetBoosterData.L2Chance
            L3Chance := TargetBoosterData.L3Chance
            L4Chance := TargetBoosterData.L4Chance
            L5Chance := TargetBoosterData.L5Chance
            L6Chance := TargetBoosterData.L6Chance

            RandomLevel := GetRandomLevel(L1Chance, L2Chance, L3Chance, L4Chance, L5Chance, L6Chance)
            ItemGranter := GetItemGranter(RandomLevel)

            race:
                TerminateLoopBehaviorEvent.Await()
                sync:
                    loop:
                        TerminateBoosterHandlerEvent.Await()
                        Logger.Print("{BoosterDisplayName}: Loop Finished", ?Level:=log_level.Warning)
                        TerminateLoopBehaviorEvent.Signal()
                    loop:
                        Sleep(0.0)

                        ItemGranter.CycleToRandomItem(Player)
                        Print("Level: {RandomLevel}")

                        BoosterLoopCompleteEvent.Signal(Player)