using { /Fortnite.com/Devices }
using { /Fortnite.com/Characters }
using { /Verse.org/Simulation }
using { /Verse.org/Assets }
using { /UnrealEngine.com/Temporary/Diagnostics }

#Booster Set Up
#-----------------------------------------------------------------------------------------------------------------

booster_upgrade_data_shield_adrenaline := class<concrete>(booster_upgrade_data):
    ExecutionsLimit<override>:int= 1 
    BoosterDescriptionA<override>:string="If health falls under\n"
    BoosterDescriptionB<override>:string="%, recover"
    BoosterDescriptionC:string=" % health \nand"
    BoosterDescriptionD:string=" % shield once"
    @editable
    ShieldPercentage:float = 0.5
    @editable
    HealthPercentageTreshold:float = 0.4

booster_default_upgrade_data_shield_adrenaline := struct:
    
    DefaultValues: []booster_upgrade_data_shield_adrenaline = array{
        Upgrade0: booster_upgrade_data_shield_adrenaline = booster_upgrade_data_shield_adrenaline{}
        Upgrade1: booster_upgrade_data_shield_adrenaline = booster_upgrade_data_shield_adrenaline{ ShieldPercentage := 0.5, ExecutionsLimit := 2 }
        Upgrade2: booster_upgrade_data_shield_adrenaline = booster_upgrade_data_shield_adrenaline{ ShieldPercentage := 1.0, ExecutionsLimit := 2 }
        Upgrade3: booster_upgrade_data_shield_adrenaline = booster_upgrade_data_shield_adrenaline{ ShieldPercentage := 1.0, ExecutionsLimit := 3 }
    }

#Booster Data
#-----------------------------------------------------------------------------------------------------------------

booster_shield_adrenaline_device := class(booster_device):

    @editable
    var BoosterUpgradeData:[]booster_upgrade_data_shield_adrenaline=booster_default_upgrade_data_shield_adrenaline{}.DefaultValues

    @editable
    VfxDevice:visual_effect_powerup_device = visual_effect_powerup_device{}
    @editable
    VfxDuration: float = 3.0

    BoosterID<override>:booster_id = booster_id.ShieldAdrenaline
    BoosterDisplayName<override>:string = "Shield Adrenaline"
    BoosterIcon<override>:texture=BoosterSystem.Textures.Icons.Legacy.T_Adrenaline

#Booster Functions
#-----------------------------------------------------------------------------------------------------------------

    GetBoosterUpgradeData(BoosterUpgradeIndex:int)<decides><transacts>:booster_upgrade_data_shield_adrenaline =
        BoosterUpgradeData[BoosterUpgradeIndex]
#==============================================================================================================================        
    GetBaseBoosterUpgradeData<override>(BoosterUpgradeIndex:int)<decides><transacts>:booster_upgrade_data=GetBoosterUpgradeData[BoosterUpgradeIndex]
#==============================================================================================================================
    GetMaxBoosterUpgradeIndex<override>():int=BoosterUpgradeData.Length
#==============================================================================================================================
    GetBoosterDescription<override>(BoosterUpgradeIndex:int):string= 
        if:
            BoosterData:=GetBoosterUpgradeData[BoosterUpgradeIndex]
        then:
            TruncatedStringA:=ToTruncatedString(BoosterData.HealthPercentageTreshold,1)
            TruncatedStringB:=ToTruncatedString(BoosterData.HealthPercentageTreshold,1) # missing Health Value
            TruncatedStringC:=ToTruncatedString(BoosterData.ShieldPercentage,1)

            return BoosterData.BoosterDescriptionA + "{TruncatedStringA}" + BoosterData.BoosterDescriptionB + "{TruncatedStringB}" + BoosterData.BoosterDescriptionC + "{TruncatedStringC}" + BoosterData.BoosterDescriptionD
        return "Booster Description"
#==============================================================================================================================
    BoosterLoopBehavior<override>(Player:player, BoosterUpgradeIndex:int, BoosterLoopCompleteEvent:event(player), TerminateBoosterHandlerEvent:event())<suspends>:void =

        if(TargetBoosterData:=GetBoosterUpgradeData[BoosterUpgradeIndex]):
        
            TerminateLoopBehaviorEvent:event()=event(){}
        
            # HealthPercentage := TargetBoosterData.HealthPercentage
            ShieldPercentage := TargetBoosterData.ShieldPercentage
            HealthPercentageThreshold := TargetBoosterData.HealthPercentageTreshold

            if(FortCharacter:=Player.GetFortCharacter[]):
                        
                race:
                    TerminateLoopBehaviorEvent.Await()
                    sync:
                        loop:
                            TerminateBoosterHandlerEvent.Await()
                            Logger.Print("{BoosterDisplayName}: Loop Finished", ?Level:=log_level.Warning)
                            TerminateLoopBehaviorEvent.Signal()
                        loop:
                            Result := FortCharacter.DamagedEvent().Await()

                            if (FortCharacter.GetHealth() < (FortCharacter.GetMaxHealth() * HealthPercentageThreshold) and FortCharacter.GetHealth() > 0.0, Agent := FortCharacter.GetAgent[]):
                                # FortCharacter.Heal(FortCharacter.GetMaxHealth() * HealthPercentage)
                                FortCharacter.SetShield(FortCharacter.GetMaxShield() * ShieldPercentage)
                                StartVfx(Agent)

                                BoosterLoopCompleteEvent.Signal(Player)

#==============================================================================================================================
    StartVfx(Agent:agent):void =
        VfxDevice.Spawn()
        VfxDevice.SetDuration(VfxDuration)
        VfxDevice.Pickup(Agent)
