using { /Fortnite.com/Devices }
using { /Fortnite.com/Characters }
using { /Verse.org/Simulation }
using { /Verse.org/Assets }
using { /UnrealEngine.com/Temporary/Diagnostics }

booster_upgrade_data_shield_regen:=class<concrete>(booster_upgrade_data):
    LifeSpawn<override>:float=120.0
    BoosterDescriptionA<override>:string=" Shield Regen"
    BoosterDescriptionB<override>:string=""
    @editable
    AmountOfShield:float=0.5
    @editable
    MaxAmountOfShield:float=0.5
    @editable
    RegenDelay:float=3.0
    @editable
    MaxLifeSpawn:float=360.0

booster_default_upgrade_data_shield_regen:=struct:

    DefaultValues:[]booster_upgrade_data_shield_regen=array{
        Upgrade0:booster_upgrade_data_shield_regen=booster_upgrade_data_shield_regen{}
        Upgrade1:booster_upgrade_data_shield_regen=booster_upgrade_data_shield_regen{AmountOfShield:=1.0, MaxAmountOfShield:=1.0}
        Upgrade2:booster_upgrade_data_shield_regen=booster_upgrade_data_shield_regen{AmountOfShield:=1.5, MaxAmountOfShield:=1.5}
        Upgrade3:booster_upgrade_data_shield_regen=booster_upgrade_data_shield_regen{AmountOfShield:=2.0, MaxAmountOfShield:=2.0}
    }
#==============================================================================================================================
booster_shield_regen_device := class(booster_device):

    @editable
    var BoosterUpgradeData:[]booster_upgrade_data_shield_regen=booster_default_upgrade_data_shield_regen{}.DefaultValues

    BoosterID<override>:booster_id=booster_id.ShieldRegen
    BoosterDisplayName<override>:string="Shield Regen"
    BoosterIcon<override>:texture=BoosterSystem.Textures.Icons.Legacy.T_IncreaseShieldRegen
    bIsStackable<override>:logic=true
#==============================================================================================================================
    GetBoosterUpgradeData(BoosterUpgradeIndex:int)<decides><transacts>:booster_upgrade_data_shield_regen=
        BoosterUpgradeData[BoosterUpgradeIndex]
#==============================================================================================================================     
    GetBaseBoosterUpgradeData<override>(BoosterUpgradeIndex:int)<decides><transacts>:booster_upgrade_data=GetBoosterUpgradeData[BoosterUpgradeIndex]
#==============================================================================================================================
    GetMaxBoosterUpgradeIndex<override>():int=BoosterUpgradeData.Length
#==============================================================================================================================
    GetBoosterDescription<override>(BoosterUpgradeIndex:int):string= 
        if:
            BoosterData:=GetBoosterUpgradeData[BoosterUpgradeIndex]
        then:
            TruncatedString:=ToTruncatedString(BoosterData.AmountOfShield,1)
            return "{TruncatedString}" + BoosterData.BoosterDescriptionA
        return "Booster Description"
#==============================================================================================================================
    BoosterLoopBehavior<override>(Player:player, BoosterUpgradeIndex:int, BoosterLoopCompleteEvent:event(player), TerminateBoosterHandlerEvent:event())<suspends>:void=
        if(TargetBoosterData:=GetBoosterUpgradeData[BoosterUpgradeIndex]):

            TerminateLoopBehaviorEvent:event()=event(){}

            var AmountOfShield:float=TargetBoosterData.AmountOfShield
            MaxAmountOfShield:=TargetBoosterData.MaxAmountOfShield
            RegenDelay:=TargetBoosterData.RegenDelay

            if(FortCharacter:=Player.GetFortCharacter[]):
                
                race:
                    TerminateLoopBehaviorEvent.Await()
                    sync:
                        loop:
                            TerminateBoosterHandlerEvent.Await()
                            Logger.Print("{BoosterDisplayName}: Loop Finished", ?Level:=log_level.Warning)
                            TerminateLoopBehaviorEvent.Signal()
                        loop:
                            EventData:=OnBoosterNewInstanceStartedEvent.Await()
                            EventPlayer:=EventData(0)
                            EventBoosterID:=EventData(1)  
                            EventBoosterUpgradeIndex:=EventData(2) 
                            if(EventPlayer = Player):
                                if(bIsStackable?):
                                    if(NewTargetBoosterData:=GetBoosterUpgradeData[BoosterUpgradeIndex]):
                                        NewAmountOfShield:=AmountOfShield + NewTargetBoosterData.AmountOfShield
                                        set AmountOfShield = Clamp(NewAmountOfShield, 0.0, MaxAmountOfShield)
                                        Logger.Print("{BoosterDisplayName}: Shield Regen Stacked: {AmountOfShield}", ?Level:=log_level.Warning)
                        loop:
                            Sleep(RegenDelay)
                            CurrentShield:=FortCharacter.GetShield()
                            NewShield:=CurrentShield+AmountOfShield
                            if(FortCharacter.GetShield() < FortCharacter.GetMaxShield()):
                                FortCharacter.SetShield(NewShield)
                            BoosterLoopCompleteEvent.Signal(Player)
#==============================================================================================================================
    OnBoosterNewInstanceStarted<override>(Player:player, NewBoosterUpgradeIndex:int, CurrentBoosterRunData:booster_run_data):void=
        if(NewTargetBoosterData:=GetBoosterUpgradeData[NewBoosterUpgradeIndex]):
            if(bIsStackable?):
                LifeSpawn:= CurrentBoosterRunData.LifeSpawnLeft + NewTargetBoosterData.LifeSpawn
                Limit:=NewTargetBoosterData.MaxLifeSpawn
                NewLifeSpawn:=Clamp(LifeSpawn, 0.0, Limit)
                Print("Updating Run Data, New Life Spawn: {NewLifeSpawn}")
                CurrentBoosterRunData.UpdateLifeSpawn(NewLifeSpawn)
                OnBoosterRunDataUpdatedEvent.Signal(CurrentBoosterRunData)