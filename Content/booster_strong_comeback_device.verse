using { /Fortnite.com/Devices }
using { /Fortnite.com/Characters }
using { /Verse.org/Simulation }
using { /Verse.org/Assets }
using { /UnrealEngine.com/Temporary/Diagnostics }

#Booster Set Up
#-----------------------------------------------------------------------------------------------------------------

booster_upgrade_data_strong_comeback := class<concrete>(booster_upgrade_data):
    BoosterDescriptionA<override>:string="Revive in the room \nyou die with "
    BoosterDescriptionB<override>:string="% health \n&"
    BoosterDescriptionC:string=" % shield"
    ExecutionsLimit<override>:int=1
    @editable
    HealthPercentage:float = 0.5
    @editable
    ShieldPercentage:float = 0.0

booster_default_upgrade_data_strong_comeback := struct:
    
    DefaultValues: []booster_upgrade_data_strong_comeback = array{
        Upgrade0: booster_upgrade_data_strong_comeback = booster_upgrade_data_strong_comeback{}
        Upgrade1: booster_upgrade_data_strong_comeback = booster_upgrade_data_strong_comeback{ShieldPercentage := 0.5 }
        Upgrade2: booster_upgrade_data_strong_comeback = booster_upgrade_data_strong_comeback{HealthPercentage := 0.5, ShieldPercentage := 1.0 }
        Upgrade3: booster_upgrade_data_strong_comeback = booster_upgrade_data_strong_comeback{HealthPercentage := 1.0, ShieldPercentage := 1.0 }
    }

#Booster Data
#-----------------------------------------------------------------------------------------------------------------

booster_strong_comeback_device := class(booster_device):

    @editable
    var BoosterUpgradeData:[]booster_upgrade_data_strong_comeback = booster_default_upgrade_data_strong_comeback{}.DefaultValues

    @editable
    VfxDevice:visual_effect_powerup_device = visual_effect_powerup_device{}
    @editable
    VfxDuration: float = 3.0

    BoosterID<override>:booster_id = booster_id.StrongComeback
    BoosterDisplayName<override>:string = "Strong Comeback"
    BoosterIcon<override>:texture=BoosterSystem.Textures.Icons.Legacy.T_CaffeinatedRevival

#Booster Functions
#-----------------------------------------------------------------------------------------------------------------

    GetBoosterUpgradeData(BoosterUpgradeIndex:int)<decides><transacts>:booster_upgrade_data_strong_comeback =
        BoosterUpgradeData[BoosterUpgradeIndex]
#==============================================================================================================================        
    GetBaseBoosterUpgradeData<override>(BoosterUpgradeIndex:int)<decides><transacts>:booster_upgrade_data=GetBoosterUpgradeData[BoosterUpgradeIndex]
#==============================================================================================================================
    GetMaxBoosterUpgradeIndex<override>():int=BoosterUpgradeData.Length
#==============================================================================================================================
    GetBoosterDescription<override>(BoosterUpgradeIndex:int):string= 
        if:
            BoosterData:=GetBoosterUpgradeData[BoosterUpgradeIndex]
        then:
            TruncatedStringA:=ToTruncatedString(BoosterData.HealthPercentage,1)
            TruncatedStringB:=ToTruncatedString(BoosterData.ShieldPercentage,1)
            return BoosterData.BoosterDescriptionA + "{TruncatedStringA}" + BoosterData.BoosterDescriptionB + "{TruncatedStringB}" + BoosterData.BoosterDescriptionC
        return "Booster Description"
#==============================================================================================================================
    BoosterLoopBehavior<override>(Player:player, BoosterUpgradeIndex:int, BoosterLoopCompleteEvent:event(player), TerminateBoosterHandlerEvent:event())<suspends>:void =

        if(TargetBoosterData:=GetBoosterUpgradeData[BoosterUpgradeIndex]):
        
            TerminateLoopBehaviorEvent:event()=event(){}
        
            HealthPercentage := TargetBoosterData.HealthPercentage
            ShieldPercentage := TargetBoosterData.ShieldPercentage

            if(FortCharacter:=Player.GetFortCharacter[]):
                        
                race:
                    TerminateLoopBehaviorEvent.Await()
                    sync:
                        loop:
                            TerminateBoosterHandlerEvent.Await()
                            Logger.Print("{BoosterDisplayName}: Loop Finished", ?Level:=log_level.Warning)
                            TerminateLoopBehaviorEvent.Signal()
                        loop:
                            Result := FortCharacter.DamagedEvent().Await()

                            if (FortCharacter.GetHealth() <= 0.0, Agent := FortCharacter.GetAgent[]):
                                FortCharacter.SetHealth(FortCharacter.GetMaxHealth() * HealthPercentage)
                                FortCharacter.SetShield(FortCharacter.GetMaxShield() * ShieldPercentage)
                                StartVfx(Agent)

                                BoosterLoopCompleteEvent.Signal(Player)
#==============================================================================================================================
    StartVfx(Agent:agent):void =
        VfxDevice.Spawn()
        VfxDevice.SetDuration(VfxDuration)
        VfxDevice.Pickup(Agent)