using { /Fortnite.com/Devices }
using { /Verse.org/Simulation }
using { /Verse.org/Assets }
using { /UnrealEngine.com/Temporary/Diagnostics }

booster_upgrade_data_swift_feet:=class<concrete>(booster_upgrade_data):
    LifeSpawn<override>:float=300.0
    BoosterDescriptionA<override>:string="x"
    BoosterDescriptionB<override>:string=" Speed Boost"
    @editable
    SpeedMultiplier:float=1.30 #+30%
    @editable
    MovementModulator:movement_modulator_device = movement_modulator_device{}

booster_default_upgrade_data_swift_feet:=struct:

    DefaultValues:[]booster_upgrade_data_swift_feet=array{
        Upgrade0:booster_upgrade_data_swift_feet=booster_upgrade_data_swift_feet{}
        Upgrade1:booster_upgrade_data_swift_feet=booster_upgrade_data_swift_feet{SpeedMultiplier:=1.5}
        Upgrade2:booster_upgrade_data_swift_feet=booster_upgrade_data_swift_feet{SpeedMultiplier:=1.8}
        Upgrade3:booster_upgrade_data_swift_feet=booster_upgrade_data_swift_feet{SpeedMultiplier:=2.0}
    }
#==============================================================================================================================
booster_swift_feet_device := class(booster_device):

    @editable
    var BoosterUpgradeData:[]booster_upgrade_data_swift_feet=booster_default_upgrade_data_swift_feet{}.DefaultValues
    @editable
    MovementModulatorNormalSpeed:movement_modulator_device = movement_modulator_device{}

    BoosterID<override>:booster_id=booster_id.SwiftFeet
    BoosterDisplayName<override>:string="Swift Feet"
    BoosterIcon<override>:texture=BoosterSystem.Textures.Icons.Legacy.T_MarathonRunner
#==============================================================================================================================
    GetBoosterUpgradeData(BoosterUpgradeIndex:int)<decides><transacts>:booster_upgrade_data_swift_feet=
        BoosterUpgradeData[BoosterUpgradeIndex]
#==============================================================================================================================       
    GetBaseBoosterUpgradeData<override>(BoosterUpgradeIndex:int)<decides><transacts>:booster_upgrade_data=GetBoosterUpgradeData[BoosterUpgradeIndex]
#==============================================================================================================================
    GetMaxBoosterUpgradeIndex<override>():int=BoosterUpgradeData.Length
#==============================================================================================================================
    GetBoosterDescription<override>(BoosterUpgradeIndex:int):string= 
        if:
            BoosterData:=GetBoosterUpgradeData[BoosterUpgradeIndex]
        then:
            TruncatedStringA:=ToTruncatedString(BoosterData.SpeedMultiplier,1)
            return BoosterData.BoosterDescriptionA + "{TruncatedStringA}" + BoosterData.BoosterDescriptionB
        return "Booster Description"
#==============================================================================================================================
    BoosterLoopBehavior<override>(Player:player, BoosterUpgradeIndex:int, BoosterLoopCompleteEvent:event(player), TerminateBoosterHandlerEvent:event())<suspends>:void=
        if(TargetBoosterData:=GetBoosterUpgradeData[BoosterUpgradeIndex]):
            TargetBoosterData.MovementModulator.Activate(Player)

            TerminateLoopBehaviorEvent:event()=event(){}

            race:
                TerminateLoopBehaviorEvent.Await()
                sync:
                    loop:
                        TerminateBoosterHandlerEvent.Await()
                        Logger.Print("{BoosterDisplayName}: Loop Finished", ?Level:=log_level.Warning)
                        MovementModulatorNormalSpeed.Activate(Player)
                        TerminateLoopBehaviorEvent.Signal()
                    loop:
                        Sleep(1.0)
                        Logger.Print("{BoosterDisplayName}: Running, Upgrade Index: {BoosterUpgradeIndex}")
                        BoosterLoopCompleteEvent.Signal(Player)
                        break





 

    