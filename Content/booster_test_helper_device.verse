
using { /Fortnite.com/Devices }
using { /Verse.org/Simulation }
using { /UnrealEngine.com/Temporary/Diagnostics }

test_booster_helper_device := class(creative_device):

    @editable
    GiveButtons:[]button_device=array{}

    @editable
    IncreaseUpgradeIndexButton:button_device=button_device{}

    @editable
    RemoveAllBoostersButton:button_device=button_device{}

    @editable
    BoosterManager:booster_manager_device=booster_manager_device{}

    var BoosterUpgradeIndex:int=0
    @editable
    BoosterUpgradeIndex_Limit:int=3
#==============================================================================================================================
    OnBegin<override>()<suspends>:void=

        #Wait one second so booster manager has enough time to find all boosters by tag
        Sleep(2.0)

        var Index:int=0
        for(GiveButton:GiveButtons):
            GiveButton.InteractedWithEvent.SubscribeAgent(OnButtonGiveBoosterInteracted,Index)
            if(TargetBooster := BoosterManager.AvailableBoosters[Index]):
                Print("Button Set: {TargetBooster.BoosterDisplayName}")
                GiveButton.SetInteractionText(StringToMessage(TargetBooster.BoosterDisplayName))
            set Index+=1

        IncreaseUpgradeIndexButton.InteractedWithEvent.Subscribe(OnButtonUpgradeIndexInteracted)
        RemoveAllBoostersButton.InteractedWithEvent.Subscribe(OnButtonRemoveAllBoostersInteracted)        
#==============================================================================================================================
    OnButtonGiveBoosterInteracted(Agent:agent, Index:int):void=
        if(Player:=player[Agent]):
           if(TargetBooster:=BoosterManager.AvailableBoosters[Index]):
            Print("Giving: {TargetBooster.BoosterDisplayName}")
            BID:=TargetBooster.BoosterID
            MaxUpgrade:=TargetBooster.GetMaxBoosterUpgradeIndex()
            ClampUpgradeValue:=Clamp(BoosterUpgradeIndex,0,MaxUpgrade)
            Print("Upgrade: {BoosterUpgradeIndex}, Max: {ClampUpgradeValue}")
            BoosterManager.GiveBoosterToPlayer(Player, BID, ClampUpgradeValue)
#==============================================================================================================================
    OnButtonUpgradeIndexInteracted(Agent:agent):void=
        if(Player:=player[Agent]):
            set BoosterUpgradeIndex += 1
            if(BoosterUpgradeIndex>=BoosterUpgradeIndex_Limit):
                set BoosterUpgradeIndex = 0
            Print("Current Booster Upgrade Index: {BoosterUpgradeIndex}")
#==============================================================================================================================
    OnButtonRemoveAllBoostersInteracted(Agent:agent):void=
        if(Player:=player[Agent]):
            Print("Removing All Boosters")
            BoosterManager.RemoveAllBoostersFromPlayer(Player)
#==============================================================================================================================
    # ConvertStringToMessage<localizes>(value:string) : message = "{value}"
#==============================================================================================================================         
# Code to be able to pass more arguments to event subscriptions. Works for events that send an agent and that don't send anything.
# (Listenable : listenable(agent)).SubscribeAgent(OutputFunc : tuple(agent, t)->void, ExtraData : t where t:type) : cancelable =
#     Wrapper := wrapper_agent(t){ExtraData := ExtraData, OutputFunc := OutputFunc}
#     Listenable.Subscribe(Wrapper.InputFunc)
    
# wrapper_agent(t : type) := class():
#     ExtraData : t;
#     OutputFunc : tuple(agent, t) -> void
#     InputFunc(Agent : agent):void = OutputFunc(Agent, ExtraData)

