using { /Fortnite.com/Devices }
using { /Fortnite.com/Characters }
using { /Verse.org/Simulation }
using { /Verse.org/Assets }
using { /UnrealEngine.com/Temporary/Diagnostics }

#Booster Set Up
#-----------------------------------------------------------------------------------------------------------------

booster_upgrade_data_vampire_bullets := class<concrete>(booster_upgrade_data):

    LifeSpawn<override>:float=120.0
    BoosterDescriptionA<override>:string="Get "
    BoosterDescriptionB<override>:string="% HP from \nDMG dealt"
    @editable
    HealingPercentage:float = 0.05

booster_default_upgrade_data_vampire_bullets:=struct:

    DefaultValues:[]booster_upgrade_data_vampire_bullets=array{
        Upgrade0:booster_upgrade_data_vampire_bullets = booster_upgrade_data_vampire_bullets{}
        Upgrade1:booster_upgrade_data_vampire_bullets = booster_upgrade_data_vampire_bullets{HealingPercentage := 0.1 }
        Upgrade2:booster_upgrade_data_vampire_bullets = booster_upgrade_data_vampire_bullets{HealingPercentage := 0.15 }
        Upgrade3:booster_upgrade_data_vampire_bullets = booster_upgrade_data_vampire_bullets{HealingPercentage := 0.20 }
    }

#Booster Data
#-----------------------------------------------------------------------------------------------------------------

booster_vampire_bullets_device := class(booster_device):

    @editable
    var BoosterUpgradeData:[]booster_upgrade_data_vampire_bullets = booster_default_upgrade_data_vampire_bullets{}.DefaultValues

    @editable
    GuardSpawner:guard_spawner_device = guard_spawner_device{}

    BoosterID<override>:booster_id = booster_id.VampireBullets
    BoosterDisplayName<override>:string = "Vampire Bullets"
    BoosterIcon<override>:texture=BoosterSystem.Textures.Icons.Legacy.T_Vampire

#Booster Functions
#-----------------------------------------------------------------------------------------------------------------

    GetBoosterUpgradeData(BoosterUpgradeIndex:int)<decides><transacts>:booster_upgrade_data_vampire_bullets =
        BoosterUpgradeData[BoosterUpgradeIndex]
        
    GetBaseBoosterUpgradeData<override>(BoosterUpgradeIndex:int)<decides><transacts>:booster_upgrade_data=GetBoosterUpgradeData[BoosterUpgradeIndex]

    GetMaxBoosterUpgradeIndex<override>():int=BoosterUpgradeData.Length
#==============================================================================================================================
    GetBoosterDescription<override>(BoosterUpgradeIndex:int):string= 
        if:
            BoosterData:=GetBoosterUpgradeData[BoosterUpgradeIndex]
        then:
            TruncatedStringA:=ToTruncatedString(BoosterData.HealingPercentage,2)
            return BoosterData.BoosterDescriptionA + "{TruncatedStringA}" + BoosterData.BoosterDescriptionB
        return "Booster Description"

    BoosterLoopBehavior<override>(Player:player, BoosterUpgradeIndex:int, BoosterLoopCompleteEvent:event(player), TerminateBoosterHandlerEvent:event())<suspends>:void =

        if(TargetBoosterData:=GetBoosterUpgradeData[BoosterUpgradeIndex]):
        
            TerminateLoopBehaviorEvent:event()=event(){}
        
            HealingPercentage := TargetBoosterData.HealingPercentage

            if(FortCharacter:=Player.GetFortCharacter[]):
                        
                race:
                    TerminateLoopBehaviorEvent.Await()
                    sync:
                        loop:
                            TerminateBoosterHandlerEvent.Await()
                            Logger.Print("{BoosterDisplayName}: Loop Finished", ?Level:=log_level.Warning)
                            TerminateLoopBehaviorEvent.Signal()
                        loop:
                            Sleep(1.0)
                            Logger.Print("{BoosterDisplayName}: Running, Upgrade Index: {BoosterUpgradeIndex}")
                            BoosterLoopCompleteEvent.Signal(Player)
                        loop:
                            #TODO: Julian -> Replace with a reference to DamagedEvent on enemy manager to get the damage value
                            Result := GuardSpawner.DamagedEvent.Await()
                            Damage := 20.0

                            if(FortCharacter.GetHealth() < FortCharacter.GetMaxHealth()):
                                FortCharacter.Heal(Damage * HealingPercentage)
