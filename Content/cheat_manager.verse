using { /Fortnite.com/Devices }
using { /Verse.org/Simulation }

<#
Device where cheats and buttons are saved
#>
cheat_manager := class(creative_device):
    @editable
    GiveCurrencyButton : button_device = button_device{}

    @editable
    CurrencyManager : clash_currency_manager = clash_currency_manager{}

    @editable
    CurrencyAmount : float = 1000000000000000000000000.0


    OnBegin<override>()<suspends>:void = 
        GiveCurrencyButton.InteractedWithEvent.Subscribe(OnGiveCurrencyButtonInteracted)

    OnGiveCurrencyButtonInteracted(Agent:agent):void =
        if (Player:=player[Agent]):
            CurrencyManager.AddCurrency(CurrencyAmount, Player, currency_id.PassiveIncome)

    OnOverridePlayerStateButtonInteracted(Agent:agent):void = {}
   