
using { /Fortnite.com/Devices }
using { /Verse.org/Simulation }
using { /UnrealEngine.com/Temporary/Diagnostics }



# A Verse-authored creative device that can be placed in a level
clash_audio_manager := class(creative_device):
    # --------------------------------------------------------------------------------
    #                                VARIABLES 
    # --------------------------------------------------------------------------------
    # -- Lobby --
    @editable
    AudioPlayer_SelectYourTeam: audio_player_device = audio_player_device{}   
    @editable
    AudioPlayer_TitleScreen: audio_player_device = audio_player_device{}
    @editable
    AudioPlayer_TeamSelected_Blue: audio_player_device = audio_player_device{}    
    @editable
    AudioPlayer_TeamSelected_Red: audio_player_device = audio_player_device{}
    @editable
    RadioPlayer_ElevatorMusic: audio_player_device = audio_player_device{}
    @editable
    AudioPlayer_MatchStart: audio_player_device = audio_player_device{}
    # -- Kill --
    @editable
    AudioPlayer_PlayerKilledRival: audio_player_device = audio_player_device{}
	@editable
    AudioPlayer_RivalKilledPlayer: audio_player_device = audio_player_device{}
	@editable
    AudioPlayer_TeamSwipe: audio_player_device = audio_player_device{}
    @editable
    AudioPlayer_RedKO_1: audio_player_device = audio_player_device{}
    @editable
    AudioPlayer_BlueKO_1: audio_player_device = audio_player_device{}
    # -- Room --
    @editable
    AudioPlayer_Blue_Room: audio_player_device = audio_player_device{}
    @editable
    AudioPlayer_Red_Room: audio_player_device = audio_player_device{}
    @editable
    RoomCapturedAudioPlayerDevice: audio_player_device = audio_player_device{}
    @editable
    RoomClashAudioPlayerDevice: audio_player_device = audio_player_device{}
    # @editable
    # EnemyRoomCapturedAudioPlayerDevice: audio_player_device = audio_player_device{}
    @editable
    EnemyEnteredRoomAudioPlayerDevice: audio_player_device = audio_player_device{}
    @editable
    EnemyLeftRoomAudioPlayerDevice: audio_player_device = audio_player_device{}
    # -- BED --
    @editable
    RedBedUnderAttackAudioPlayerDevice: audio_player_device = audio_player_device{}
    @editable
    BlueBedUnderAttackAudioPlayerDevice: audio_player_device = audio_player_device{}
    @editable
    RedBedDestroyedAudioPlayerDevice: audio_player_device = audio_player_device{}
    @editable
    BlueBedDestroyedAudioPlayerDevice: audio_player_device = audio_player_device{}
    # -- SCORE --
    @editable
    RedVictoryAudioPlayerDevice: audio_player_device = audio_player_device{}
    @editable
    BlueVictoryAudioPlayerDevice: audio_player_device = audio_player_device{}
    # --------------------------------------------------------------------------------
    #                                FUNCTIONS 
    # --------------------------------------------------------------------------------
    
    
    # -- Lobby Audio --    
    PlayAudio_TitleScreen(Agent : agent):void =
        AudioPlayer_TitleScreen.Play(Agent)
    PlayAudio_MatchStart():void =
        Players := GetPlayspace().GetPlayers()
            for(Player:Players):
                AudioPlayer_MatchStart.Play(Player)
    PlayAudio_TeamSelectedBlue(Agent : agent):void =
        AudioPlayer_TeamSelected_Blue.Play(Agent)    
    PlayAudio_TeamSelectedRed(Agent : agent):void =
        AudioPlayer_TeamSelected_Red.Play(Agent)
    PlayAudio_SelectYourTeam(Agent : agent):void =
        AudioPlayer_SelectYourTeam.Play(Agent)
    PlayAudio_ElevatorMusic(Agent : agent, ShouldStop : logic):void =
        if (ShouldStop = true):
            Players := GetPlayspace().GetPlayers()
            for(Player:Players):
                RadioPlayer_ElevatorMusic.Stop(Player)            
        else:    
            RadioPlayer_ElevatorMusic.Play(Agent)
    # -- Kill --
    PlayAudio_PlayerKilledRival(Agent : agent):void =
        AudioPlayer_PlayerKilledRival.Play(Agent)
    PlayAudio_RivalKilledPlayer(Agent : agent):void =
        AudioPlayer_RivalKilledPlayer.Play(Agent)
    PlayAudio_TeamSwipe(Agent : agent):void =
        AudioPlayer_TeamSwipe.Play(Agent)        
    PlayAudio_RedKO_1(Agent : agent):void =
        AudioPlayer_RedKO_1.Play(Agent)
    PlayAudio_BlueKO_1(Agent : agent):void =
        AudioPlayer_BlueKO_1.Play(Agent)
    # -- Room --
    PlayAudio_Blue_Room(Agent : agent):void =
        AudioPlayer_Blue_Room.Play(Agent)
    PlayAudio_Red_Room(Agent : agent):void =
        AudioPlayer_Red_Room.Play(Agent)
    PlayAudio_CapturedRoom(Agent : agent):void =
        RoomCapturedAudioPlayerDevice.Play(Agent)
    PlayAudio_EnemyEntered(Agent : agent):void =
        EnemyEnteredRoomAudioPlayerDevice.Play(Agent)
    PlayAudio_EnemyLeft(Agent : agent):void =
        EnemyLeftRoomAudioPlayerDevice.Play(Agent)
    PlayAudio_RoomClash(Agent : agent):void =
        RoomClashAudioPlayerDevice.Play(Agent)
    # -- BED -- 
    PlayAudio_RedBedUnderAttack(Agent : agent):void =
        RedBedUnderAttackAudioPlayerDevice.Play(Agent)
    PlayAudio_BlueBedUnderAttack(Agent : agent):void =
        BlueBedUnderAttackAudioPlayerDevice.Play(Agent)
    PlayAudio_RedBedDestroyed(Agent : agent):void =
        RedBedDestroyedAudioPlayerDevice.Play(Agent)
    PlayAudio_BlueBedDestroyed(Agent : agent):void =
        BlueBedDestroyedAudioPlayerDevice.Play(Agent)
    # -- SCORE -- 
    PlayAudio_RedVictory(Agent : agent):void =
        RedVictoryAudioPlayerDevice.Play(Agent)
    PlayAudio_BlueVictory(Agent : agent):void =
        BlueVictoryAudioPlayerDevice.Play(Agent)
