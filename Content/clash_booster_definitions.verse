using { /Verse.org/Assets }

clash_booster_reward:=class(rewardable,iconTextureable,describable):
    BoosterID:booster_id=booster_id.SwiftFeet
    BoosterIcon:texture = BoosterSystem.Textures.Icons.T_UI_BSTR_Blank
    UpgradeIndex:int=0
    BoosterDisplayName:string = "" 
    BoosterDescription:string = ""
    GetRarity<override>()<computes>: item_rarity = item_rarity.Common #TODO: @Cbray: Create method to compute UpgradeIndex to Rarity Type
    GetDisplayName<override>()<computes>: string = BoosterDisplayName  
    GetIconTexture<override>()<computes>: texture = BoosterIcon  
    GetDescription<override>()<transacts>: string = BoosterDescription