using { /Fortnite.com/Devices }
using { /Fortnite.com/Characters }
using { /Verse.org/Simulation }
using { /Verse.org/Assets }
using { /UnrealEngine.com/Temporary/Diagnostics }
using { /Fortnite.com/Game }
using { /UnrealEngine.com/Temporary/SpatialMath }


clash_booster_default_upgrade_data_explosive_enemies:=struct:

    DefaultValues:[]booster_upgrade_data_explosive_enemies = array{
        Upgrade0:booster_upgrade_data_explosive_enemies = booster_upgrade_data_explosive_enemies { }
        Upgrade1:booster_upgrade_data_explosive_enemies = booster_upgrade_data_explosive_enemies { ExplosionRadius:= 400.0 }
        Upgrade2:booster_upgrade_data_explosive_enemies = booster_upgrade_data_explosive_enemies { ExplosionRadius:= 550.0 }
        Upgrade3:booster_upgrade_data_explosive_enemies = booster_upgrade_data_explosive_enemies { ExplosionRadius:= 800.0 }
    }

#-----------------------------------------------------------------------------------------------------------------

clash_booster_explosive_enemies_device := class(booster_explosive_enemies_device):

    @editable
    var BoosterUpgradeData<override>:[]booster_upgrade_data_explosive_enemies = clash_booster_default_upgrade_data_explosive_enemies{}.DefaultValues