using { /Fortnite.com/Devices }
using { /Fortnite.com/Characters }
using { /Verse.org/Simulation }
using { /Verse.org/Assets }
using { /UnrealEngine.com/Temporary/Diagnostics }



clash_booster_default_upgrade_data_fire_bullets:=struct:

    DefaultValues:[]booster_upgrade_data_fire_bullets=array{
        Upgrade0:booster_upgrade_data_fire_bullets=booster_upgrade_data_fire_bullets{}
        Upgrade1:booster_upgrade_data_fire_bullets=booster_upgrade_data_fire_bullets{Damage:=2.0, DamageWindow:=10.0, MaxDamageWindow:=20.0}
        Upgrade2:booster_upgrade_data_fire_bullets=booster_upgrade_data_fire_bullets{Damage:=3.0, DamageWindow:=20.0, MaxDamageWindow:=40.0}
        Upgrade3:booster_upgrade_data_fire_bullets=booster_upgrade_data_fire_bullets{Damage:=5.0, DamageWindow:=30.0, MaxDamageWindow:=60.0}
    }
#========================================================================================================================= 
clash_booster_fire_bullets_device := class(booster_fire_bullets_device):

    @editable
    var BoosterUpgradeData<override>:[]booster_upgrade_data_fire_bullets=clash_booster_default_upgrade_data_fire_bullets{}.DefaultValues

    @editable
    ClashEnemiesManager:clash_enemies_manager=clash_enemies_manager{}  
#========================================================================================================================= 
    AwaitForEnemyDamage<override>()<suspends>:void=
        if(bDebugEnabled?){Logger.Print("Calling AwaitForEnemyDamage")}
        loop:
            Result:= ClashEnemiesManager.EnemyReceivedDamageEvent.Await()
            if(bDebugEnabled?){Logger.Print("Enemy Damage Detected")}
            var MaybeSourceFortCharacter:?fort_character=false
            if:
                Target:=Result.Target
                TargetFortCharacter:=fort_character[Target] 
            then:
                if:
                    Instigator:=Result.Instigator?
                    InstigatorAgent:=Instigator.GetInstigatorAgent[]
                    SourceFortCharacter:=InstigatorAgent.GetFortCharacter[] 
                then:
                    set MaybeSourceFortCharacter = option{SourceFortCharacter}
                OnEnemyDamagedEvent.Signal(TargetFortCharacter, MaybeSourceFortCharacter)