
using { /Fortnite.com/Devices }
using { /Verse.org/Simulation }
using { /UnrealEngine.com/Temporary/Diagnostics }
using { /Verse.org/Assets }

clash_booster_default_upgrade_data_gold_digger := struct:

    DefaultValues: []booster_upgrade_data_gold_digger=array{
        Upgrade0:booster_upgrade_data_gold_digger = booster_upgrade_data_gold_digger{LifeSpawn := 60.0}
        Upgrade1:booster_upgrade_data_gold_digger = booster_upgrade_data_gold_digger{LifeSpawn := 60.0, ExtraGold := 6.0}
        Upgrade2:booster_upgrade_data_gold_digger = booster_upgrade_data_gold_digger{LifeSpawn := 60.0, ExtraGold := 8.0}
        Upgrade3:booster_upgrade_data_gold_digger = booster_upgrade_data_gold_digger{LifeSpawn := 60.0, ExtraGold := 10.0}
        Upgrade4:booster_upgrade_data_gold_digger = booster_upgrade_data_gold_digger{LifeSpawn := 60.0, ExtraGold := 12.0}
    }

#-----------------------------------------------------------------------------------------------------------------------

clash_booster_gold_digger_device := class(booster_gold_digger_device):

    @editable
    CurrencyManager : clash_currency_manager = clash_currency_manager{}

    @editable
    ClashEnemiesManager : clash_enemies_manager = clash_enemies_manager{}

    @editable
    var BoosterUpgradeData<override>: []booster_upgrade_data_gold_digger = clash_booster_default_upgrade_data_gold_digger{}.DefaultValues

#-----------------------------------------------------------------------------------------------------------------------

    GetBoosterUpgradeData<override>(BoosterUpgradeIndex:int)<decides><transacts>:booster_upgrade_data_gold_digger=
        BoosterUpgradeData[BoosterUpgradeIndex]
    
    GetBaseBoosterUpgradeData<override>(BoosterUpgradeIndex:int)<decides><transacts>:booster_upgrade_data=GetBoosterUpgradeData[BoosterUpgradeIndex]

    BoosterLoopBehavior<override>(Player:player, BoosterUpgradeIndex:int, BoosterLoopCompleteEvent:event(player), TerminateBoosterHandlerEvent:event())<suspends>:void=
        <#
        @Player: Player running this Booster
        @BoosterLoopCompleteEvent: if Loop is Complete Call the event to indicate booster was used
        @TerminateBoosterHandlerEvent.Await(): to finish this method, this is call from the main handler, if time is out or the execution limit was reached
        @TerminateLoopBehaviorEvent.Await() : To Finish this method and all sub rutines
        #>

        if(TargetBoosterData:=GetBoosterUpgradeData[BoosterUpgradeIndex]):

            TerminateLoopBehaviorEvent:event()=event(){}

            GoldAmount := TargetBoosterData.ExtraGold

            race:
              TerminateLoopBehaviorEvent.Await()
              sync:
                  loop:
                      TerminateBoosterHandlerEvent.Await()
                      Logger.Print("{BoosterDisplayName}: Loop Finished", ?Level:=log_level.Warning)
                      TerminateLoopBehaviorEvent.Signal()
                  loop:
    
                      EventInfo := ClashEnemiesManager.EnemyEliminatedEvent.Await()
                      Interaction := EventInfo.Interaction    
    
                      if (Agent := Interaction.Source?, NPCKiller := player[Agent]):
                        if (NPCKiller = Player):
                            Print(GetCurrencyInfo(Player, GoldAmount))

                            CurrencyManager.AddCurrency(GoldAmount, Player, currency_id.CoinReward)
    
                      BoosterLoopCompleteEvent.Signal(Player)
