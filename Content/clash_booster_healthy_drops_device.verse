using { /Fortnite.com/Devices }
using { /Fortnite.com/Characters }
using { /Verse.org/Simulation }
using { /Verse.org/Assets }
using { /UnrealEngine.com/Temporary/Diagnostics }
using { /UnrealEngine.com/Temporary/SpatialMath }
using { /Verse.org/Random }

#Booster Set Up
#-----------------------------------------------------------------------------------------------------------------

clash_booster_default_upgrade_data_healthy_drops:=struct:

    DefaultValues:[]booster_upgrade_data_healthy_drops = array{
        Upgrade0:booster_upgrade_data_healthy_drops = booster_upgrade_data_healthy_drops{}
        Upgrade1:booster_upgrade_data_healthy_drops = booster_upgrade_data_healthy_drops{DropChance := 10}
        Upgrade2:booster_upgrade_data_healthy_drops = booster_upgrade_data_healthy_drops{DropChance := 20}
        Upgrade3:booster_upgrade_data_healthy_drops = booster_upgrade_data_healthy_drops{DropChance := 30}
    }

#Booster Data
#-----------------------------------------------------------------------------------------------------------------

clash_booster_healthy_drops_device := class(booster_healthy_drops_device):

    @editable
    var BoosterUpgradeData<override>:[]booster_upgrade_data_healthy_drops = clash_booster_default_upgrade_data_healthy_drops{}.DefaultValues

    @editable
    ClashEnemiesManager:clash_enemies_manager = clash_enemies_manager{}


#Booster Functions
#-----------------------------------------------------------------------------------------------------------------

    BoosterLoopBehavior<override>(Player:player, BoosterUpgradeIndex:int, BoosterLoopCompleteEvent:event(player), TerminateBoosterHandlerEvent:event())<suspends>:void =

        if(TargetBoosterData:=GetBoosterUpgradeData[BoosterUpgradeIndex]):
        
            TerminateLoopBehaviorEvent:event()=event(){}

            DropChance := TargetBoosterData.DropChance

            if(FortCharacter:=Player.GetFortCharacter[]):
                        
                race:
                    TerminateLoopBehaviorEvent.Await()
                    sync:
                        loop:
                            TerminateBoosterHandlerEvent.Await()
                            Logger.Print("{BoosterDisplayName}: Loop Finished", ?Level:=log_level.Warning)
                            TerminateLoopBehaviorEvent.Signal()
                        loop:
                            Sleep(1.0)
                            Logger.Print("{BoosterDisplayName}: Running, Upgrade Index: {BoosterUpgradeIndex}")
                            BoosterLoopCompleteEvent.Signal(Player)
                        loop:
                            Result := ClashEnemiesManager.EnemyEliminatedEvent.Await()
                            QAgent := Result.Interaction.Target

                            if:
                                QAgent <> false
                                Agent: agent = QAgent?
                                EliminatedCharacter: fort_character = Agent.GetFortCharacter[]
                                EliminatedTransform := EliminatedCharacter.GetTransform()
                                RandomGeneratedInt := GetRandomInt(0, 100)
                            then:
                                if(RandomGeneratedInt <= DropChance):
                                    if(ItemSpawner.TeleportTo[EliminatedTransform.Translation + vector3{Z:=0.002}, EliminatedTransform.Rotation]):
                                        ItemSpawner.SpawnItem()
