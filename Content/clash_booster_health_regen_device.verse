using { /Fortnite.com/Devices }
using { /Fortnite.com/Characters }
using { /Verse.org/Simulation }
using { /Verse.org/Assets }
using { /UnrealEngine.com/Temporary/Diagnostics }

clash_booster_default_upgrade_data_health_regen:=struct:

    DefaultValues:[]booster_upgrade_data_health_regen=array{
        Upgrade0:booster_upgrade_data_health_regen=booster_upgrade_data_health_regen{}
        Upgrade1:booster_upgrade_data_health_regen=booster_upgrade_data_health_regen{AmountOfHealth:=1.0,MaxAmountOfHealth:=1.0}
        Upgrade2:booster_upgrade_data_health_regen=booster_upgrade_data_health_regen{AmountOfHealth:=1.5,MaxAmountOfHealth:=1.5}
        Upgrade3:booster_upgrade_data_health_regen=booster_upgrade_data_health_regen{AmountOfHealth:=2.0,MaxAmountOfHealth:=2.0}
    }


clash_booster_health_regen_device := class(booster_health_regen_device):

    @editable
    var BoosterUpgradeData<override>:[]booster_upgrade_data_health_regen=clash_booster_default_upgrade_data_health_regen{}.DefaultValues
