using { /Verse.org/Simulation }
using { /Verse.org/Random }

clash_booster_manager_device := class(booster_manager_device, awardable):
    
    OnBegin<override>()<suspends>:void=
        Sleep(9.0)
        (super:)OnBegin()

    GetRewardableData<override>(RewardableData:rewardable_data): rewardable =
        RandomNumber := GetRandomInt(0, booster_id_array.Length - 1)
        if:
            BoostereID := booster_id_array[RandomNumber]                   
        then:
            MaybeBooster := GetBoosterDeviceByID(BoostereID)   
            if:
                CurrentBooster := MaybeBooster?
            then:
                BoosterID:=CurrentBooster.BoosterID
                BoosterIcon:= CurrentBooster.BoosterIcon
                UpgradeIndex:int=0 #TODO: @Cbray Connect here the Player Data System
                BoosterDisplayName:string = CurrentBooster.BoosterDisplayName 
                BoosterDescription:string = CurrentBooster.GetBoosterDescription(UpgradeIndex)

                Logger.Print("{CurrentBooster.BoosterDisplayName}")
                RewardData:=clash_booster_reward{BoosterID:=BoosterID, BoosterIcon:=BoosterIcon,
                    UpgradeIndex:=UpgradeIndex,BoosterDisplayName:=BoosterDisplayName, BoosterDescription:=BoosterDescription}

                return RewardData  
            else:
                Logger.Print("Booster not found")                  
        return clash_booster_reward{}
                 
    GiveReward<override>(Player:player, Rewardable:rewardable): void =
        if:
            BoosterData := clash_booster_reward[Rewardable]     
        then:
            BoosterID := BoosterData.BoosterID     
            MaybeBooster := GetBoosterDeviceByID(BoosterID)   
            if:
                CurrentBooster := MaybeBooster?
            then:
                UpgradeIndex := BoosterData.UpgradeIndex    
                GiveBoosterToPlayer(Player, BoosterID, UpgradeIndex)