using { /Fortnite.com/Devices }
using { /Fortnite.com/Characters }
using { /Verse.org/Simulation }
using { /Verse.org/Assets }
using { /UnrealEngine.com/Temporary/Diagnostics }

clash_booster_default_upgrade_data_max_health:=struct:

    DefaultValues:[]booster_upgrade_data_max_health=array{
        Upgrade0:booster_upgrade_data_max_health=booster_upgrade_data_max_health{}
        Upgrade1:booster_upgrade_data_max_health=booster_upgrade_data_max_health{AmountOfHealth:=15.0}
        Upgrade2:booster_upgrade_data_max_health=booster_upgrade_data_max_health{AmountOfHealth:=25.0}
        Upgrade3:booster_upgrade_data_max_health=booster_upgrade_data_max_health{AmountOfHealth:=35.0}
    }


clash_booster_max_health_device := class(booster_max_health_device):

    @editable
    var BoosterUpgradeData<override>:[]booster_upgrade_data_max_health=clash_booster_default_upgrade_data_max_health{}.DefaultValues
