using { /Fortnite.com/Devices }
using { /Fortnite.com/Characters }
using { /Verse.org/Simulation }
using { /Verse.org/Assets }
using { /UnrealEngine.com/Temporary/Diagnostics }

clash_booster_default_upgrade_data_max_shield:=struct:

    DefaultValues:[]booster_upgrade_data_max_shield=array{
        Upgrade0:booster_upgrade_data_max_shield=booster_upgrade_data_max_shield{}
        Upgrade1:booster_upgrade_data_max_shield=booster_upgrade_data_max_shield{AmountOfShield:=15.0}
        Upgrade2:booster_upgrade_data_max_shield=booster_upgrade_data_max_shield{AmountOfShield:=25.0}
        Upgrade3:booster_upgrade_data_max_shield=booster_upgrade_data_max_shield{AmountOfShield:=35.0}
    }

clash_booster_max_shield_device := class(booster_max_shield_device):

    @editable
    var BoosterUpgradeData<override>:[]booster_upgrade_data_max_shield=clash_booster_default_upgrade_data_max_shield{}.DefaultValues