
using { /Fortnite.com/Devices }
using { /Verse.org/Simulation }
using { /UnrealEngine.com/Temporary/Diagnostics }
using { /Verse.org/Assets }
using { /Verse.org/Random}

#-----------------------------------------------------------------------------------------------------------------------

clash_booster_default_upgrade_data_mystery_item := struct:

    DefaultValues: []booster_upgrade_data_mystery_item=array{
        Upgrade0:booster_upgrade_data_mystery_item = booster_upgrade_data_mystery_item{}
        Upgrade1:booster_upgrade_data_mystery_item = booster_upgrade_data_mystery_item{L1Chance := 10, L2Chance := 45, L3Chance := 40, L4Chance := 5}
        Upgrade2:booster_upgrade_data_mystery_item = booster_upgrade_data_mystery_item{L1Chance := 0, L2Chance := 15, L3Chance := 35, L4Chance := 20, L5Chance := 25, L6Chance := 5}
    }

#-----------------------------------------------------------------------------------------------------------------------

clash_booster_mystery_item_device := class(booster_mystery_item_device):

    @editable
    var BoosterUpgradeData<override>: []booster_upgrade_data_mystery_item = clash_booster_default_upgrade_data_mystery_item{}.DefaultValues

    BoosterID<override>:booster_id = booster_id.MysteryItem
    BoosterDisplayName<override>:string = "Mystery Item"
    BoosterIcon<override>:texture = BoosterSystem.Textures.Icons.Legacy.T_CoinRewards #Temporal icon

#-----------------------------------------------------------------------------------------------------------------------