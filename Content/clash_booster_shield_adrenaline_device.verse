using { /Fortnite.com/Devices }
using { /Fortnite.com/Characters }
using { /Verse.org/Simulation }
using { /Verse.org/Assets }
using { /UnrealEngine.com/Temporary/Diagnostics }

#Booster Set Up
#-----------------------------------------------------------------------------------------------------------------
clash_booster_default_upgrade_data_shield_adrenaline := struct:
    
    DefaultValues: []booster_upgrade_data_shield_adrenaline = array{
        Upgrade0: booster_upgrade_data_shield_adrenaline = booster_upgrade_data_shield_adrenaline{}
        Upgrade1: booster_upgrade_data_shield_adrenaline = booster_upgrade_data_shield_adrenaline{ ShieldPercentage := 0.5, ExecutionsLimit := 2 }
        Upgrade2: booster_upgrade_data_shield_adrenaline = booster_upgrade_data_shield_adrenaline{ ShieldPercentage := 1.0, ExecutionsLimit := 2 }
        Upgrade3: booster_upgrade_data_shield_adrenaline = booster_upgrade_data_shield_adrenaline{ ShieldPercentage := 1.0, ExecutionsLimit := 3 }
    }

#Booster Data
#-----------------------------------------------------------------------------------------------------------------
clash_booster_shield_adrenaline_device := class(booster_shield_adrenaline_device):

    @editable
    var BoosterUpgradeData<override>:[]booster_upgrade_data_shield_adrenaline=clash_booster_default_upgrade_data_shield_adrenaline{}.DefaultValues


