using { /Fortnite.com/Devices }
using { /Fortnite.com/Characters }
using { /Verse.org/Simulation }
using { /Verse.org/Assets }
using { /UnrealEngine.com/Temporary/Diagnostics }

clash_booster_default_upgrade_data_shield_regen:=struct:

    DefaultValues:[]booster_upgrade_data_shield_regen=array{
        Upgrade0:booster_upgrade_data_shield_regen=booster_upgrade_data_shield_regen{}
        Upgrade1:booster_upgrade_data_shield_regen=booster_upgrade_data_shield_regen{AmountOfShield:=1.0, MaxAmountOfShield:=1.0}
        Upgrade2:booster_upgrade_data_shield_regen=booster_upgrade_data_shield_regen{AmountOfShield:=1.5, MaxAmountOfShield:=1.5}
        Upgrade3:booster_upgrade_data_shield_regen=booster_upgrade_data_shield_regen{AmountOfShield:=2.0, MaxAmountOfShield:=2.0}
    }


clash_booster_shield_regen_device := class(booster_shield_regen_device):

    @editable
    var BoosterUpgradeData<override>:[]booster_upgrade_data_shield_regen=clash_booster_default_upgrade_data_shield_regen{}.DefaultValues
