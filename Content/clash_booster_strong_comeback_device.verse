using { /Fortnite.com/Devices }
using { /Fortnite.com/Characters }
using { /Verse.org/Simulation }
using { /Verse.org/Assets }
using { /UnrealEngine.com/Temporary/Diagnostics }

#Booster Set Up
#-----------------------------------------------------------------------------------------------------------------
clash_booster_default_upgrade_data_strong_comeback := struct:
    
    DefaultValues: []booster_upgrade_data_strong_comeback = array{
        Upgrade0: booster_upgrade_data_strong_comeback = booster_upgrade_data_strong_comeback{}
        Upgrade1: booster_upgrade_data_strong_comeback = booster_upgrade_data_strong_comeback{ShieldPercentage := 0.5 }
        Upgrade2: booster_upgrade_data_strong_comeback = booster_upgrade_data_strong_comeback{HealthPercentage := 0.5, ShieldPercentage := 1.0 }
        Upgrade3: booster_upgrade_data_strong_comeback = booster_upgrade_data_strong_comeback{HealthPercentage := 1.0, ShieldPercentage := 1.0 }
    }

#Booster Data
#-----------------------------------------------------------------------------------------------------------------

clash_booster_strong_comeback_device := class(booster_strong_comeback_device):

    @editable
    var BoosterUpgradeData<override>:[]booster_upgrade_data_strong_comeback = clash_booster_default_upgrade_data_strong_comeback{}.DefaultValues