using { /Fortnite.com/Devices }
using { /Verse.org/Simulation }
using { /Verse.org/Assets }
using { /UnrealEngine.com/Temporary/Diagnostics }

clash_booster_default_upgrade_data_swift_feet:=struct:

    DefaultValues:[]booster_upgrade_data_swift_feet=array{
        Upgrade0:booster_upgrade_data_swift_feet=booster_upgrade_data_swift_feet{}
        Upgrade1:booster_upgrade_data_swift_feet=booster_upgrade_data_swift_feet{SpeedMultiplier:=1.5}
        Upgrade2:booster_upgrade_data_swift_feet=booster_upgrade_data_swift_feet{SpeedMultiplier:=1.8}
        Upgrade3:booster_upgrade_data_swift_feet=booster_upgrade_data_swift_feet{SpeedMultiplier:=2.0}
    }

clash_booster_swift_feet_device := class(booster_swift_feet_device):

    @editable
    var BoosterUpgradeData<override>:[]booster_upgrade_data_swift_feet=clash_booster_default_upgrade_data_swift_feet{}.DefaultValues
