using { /Fortnite.com/Devices }
using { /Fortnite.com/Characters }
using { /Verse.org/Simulation }
using { /Verse.org/Assets }
using { /UnrealEngine.com/Temporary/Diagnostics }

#Booster Set Up
#-----------------------------------------------------------------------------------------------------------------

clash_booster_default_upgrade_data_vampire_bullets:=struct:

    DefaultValues:[]booster_upgrade_data_vampire_bullets=array{
        Upgrade0:booster_upgrade_data_vampire_bullets = booster_upgrade_data_vampire_bullets{}
        Upgrade1:booster_upgrade_data_vampire_bullets = booster_upgrade_data_vampire_bullets{HealingPercentage := 0.1 }
        Upgrade2:booster_upgrade_data_vampire_bullets = booster_upgrade_data_vampire_bullets{HealingPercentage := 0.15 }
        Upgrade3:booster_upgrade_data_vampire_bullets = booster_upgrade_data_vampire_bullets{HealingPercentage := 0.20 }
    }

#Booster Data
#-----------------------------------------------------------------------------------------------------------------

clash_booster_vampire_bullets_device := class(booster_vampire_bullets_device):

    @editable
    var BoosterUpgradeData<override>:[]booster_upgrade_data_vampire_bullets = clash_booster_default_upgrade_data_vampire_bullets{}.DefaultValues

    @editable
    ClashEnemiesManager:clash_enemies_manager = clash_enemies_manager{}

#Booster Functions
#-----------------------------------------------------------------------------------------------------------------

    BoosterLoopBehavior<override>(Player:player, BoosterUpgradeIndex:int, BoosterLoopCompleteEvent:event(player), TerminateBoosterHandlerEvent:event())<suspends>:void =

        if(TargetBoosterData:=GetBoosterUpgradeData[BoosterUpgradeIndex]):
        
            TerminateLoopBehaviorEvent:event()=event(){}
        
            HealingPercentage := TargetBoosterData.HealingPercentage

            if(FortCharacter:=Player.GetFortCharacter[]):
                        
                race:
                    TerminateLoopBehaviorEvent.Await()
                    sync:
                        loop:
                            TerminateBoosterHandlerEvent.Await()
                            Logger.Print("{BoosterDisplayName}: Loop Finished", ?Level:=log_level.Warning)
                            TerminateLoopBehaviorEvent.Signal()
                        loop:
                            Sleep(1.0)
                            Logger.Print("{BoosterDisplayName}: Running, Upgrade Index: {BoosterUpgradeIndex}")
                            BoosterLoopCompleteEvent.Signal(Player)
                        loop:
                            Result := ClashEnemiesManager.EnemyReceivedDamageEvent.Await()
                            Damage := Result.Amount

                            if(FortCharacter.GetHealth() < FortCharacter.GetMaxHealth()):
                                FortCharacter.Heal(Damage * HealingPercentage)
