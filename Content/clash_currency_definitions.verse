
using { /Fortnite.com/Devices }
using { /Verse.org/Simulation }
using { /Verse.org/Assets }
using { /UnrealEngine.com/Temporary/Diagnostics }

currency_id := enum:
    Type1EnemyKill
    Type2EnemyKill
    Type3EnemyKill
    Type4EnemyKill
    Type5EnemyKill
    PlayerKill
    Type1EnemyAssist
    Type2EnemyAssist
    Type3EnemyAssist
    Type4EnemyAssist
    Type5EnemyAssist
    PlayerAssist
    CoinReward
    ObjectiveDestroyed
    RoomCapture
    Underdog
    PassiveIncome
    None

#-----------------------------------------------------------------------------------------------------------------------
currency_id_array:[]currency_id = array{
    currency_id.Type1EnemyKill
    currency_id.Type2EnemyKill
    currency_id.Type3EnemyKill
    currency_id.Type4EnemyKill
    currency_id.Type5EnemyKill
    currency_id.PlayerKill
    currency_id.Type1EnemyAssist
    currency_id.Type2EnemyAssist
    currency_id.Type3EnemyAssist
    currency_id.Type4EnemyAssist
    currency_id.Type5EnemyAssist
    currency_id.PlayerAssist
    currency_id.CoinReward
    currency_id.ObjectiveDestroyed
    currency_id.RoomCapture
    currency_id.Underdog
    currency_id.PassiveIncome
}

#-----------------------------------------------------------------------------------------------------------------------
GetEnemyCurrencyIdByLevel(Level:int): currency_id=
    case (Level):
        1 => currency_id.Type1EnemyKill
        2 => currency_id.Type2EnemyKill
        3 => currency_id.Type3EnemyKill
        4 => currency_id.Type4EnemyKill
        5 => currency_id.Type5EnemyKill

        _=> currency_id.None

#-----------------------------------------------------------------------------------------------------------------------
GetEnemyAssistIdByLevel(Level:int): currency_id=
    case (Level):
        1 => currency_id.Type1EnemyAssist
        2 => currency_id.Type2EnemyAssist
        3 => currency_id.Type3EnemyAssist
        4 => currency_id.Type4EnemyAssist
        5 => currency_id.Type5EnemyAssist

        _=> currency_id.None

#-----------------------------------------------------------------------------------------------------------------------
ToString(Id:currency_id) : string =
    case(Id):
        currency_id.Type1EnemyKill => "Type1EnemyKill"
        currency_id.Type2EnemyKill => "Type2EnemyKill"
        currency_id.Type3EnemyKill => "Type3EnemyKill"
        currency_id.Type4EnemyKill => "Type4EnemyKill"
        currency_id.Type5EnemyKill => "Type5EnemyKill"
        currency_id.PlayerKill => "PlayerKill"
        currency_id.Type1EnemyAssist => "Type1EnemyAssist"
        currency_id.Type2EnemyAssist => "Type2EnemyAssist"
        currency_id.Type3EnemyAssist => "Type3EnemyAssist"
        currency_id.Type4EnemyAssist => "Type4EnemyAssist"
        currency_id.Type5EnemyAssist => "Type5EnemyAssist"
        currency_id.PlayerAssist => "PlayerAssist"
        currency_id.CoinReward => "CoinReward"
        currency_id.ObjectiveDestroyed => "ObjectiveDestroyed"
        currency_id.RoomCapture => "RoomCapture"
        currency_id.Underdog => "Underdog"
        currency_id.PassiveIncome => "PassiveIncome"
        currency_id.None => "None"

#-----------------------------------------------------------------------------------------------------------------------
<# ... Add here the textures for the kill icon ...#>
FindKillSourceIcon(Id:currency_id) : texture =
    case(Id):
        currency_id.Type1EnemyKill => Textures.Boosters.T_EliminationBW
        currency_id.Type2EnemyKill => Textures.Boosters.T_EliminationBW
        currency_id.Type3EnemyKill => Textures.Boosters.T_EliminationBW
        currency_id.Type4EnemyKill => Textures.Boosters.T_EliminationBW
        currency_id.Type5EnemyKill => Textures.Boosters.T_EliminationBW
        currency_id.PlayerKill => Textures.Boosters.T_EliminationBW
        currency_id.Type1EnemyAssist => Textures.Boosters.T_AssistBW
        currency_id.Type2EnemyAssist => Textures.Boosters.T_AssistBW
        currency_id.Type3EnemyAssist => Textures.Boosters.T_AssistBW
        currency_id.Type4EnemyAssist => Textures.Boosters.T_AssistBW
        currency_id.Type5EnemyAssist => Textures.Boosters.T_AssistBW
        currency_id.PlayerAssist => Textures.Boosters.T_AssistBW
        currency_id.CoinReward => Textures.Boosters.T_CoinRewardsBW
        currency_id.ObjectiveDestroyed => Textures.Boosters.T_CoinRewardsBW
        currency_id.RoomCapture => Textures.Boosters.T_CoinRewardsBW
        currency_id.Underdog => Textures.Boosters.T_CoinRewardsBW
        currency_id.PassiveIncome => Textures.Rescue.T_AFK_ZONE_ICON_BW
        # ...

        _ => Textures.T_Blank        