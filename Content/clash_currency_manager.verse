
using { /Fortnite.com/Devices }
using { /Verse.org/Simulation }
using { /UnrealEngine.com/Temporary/Diagnostics }
using { /Fortnite.com/Playspaces }
using { /Fortnite.com/Characters}
using { /Fortnite.com/Game }
using { /Verse.org/Simulation/Tags }

currency_manager_tag:=class(tag){}

currency_changed_payload := struct:
    Player:player
    PreviousValue:float
    CurrentValue:float
    CurrencySourceIcon:currency_id

#-----------------------------------------------------------------------------------------------------------------------
clash_currency_manager := class(creative_device):

    var PlayersMap: [player] float = map{}

    @editable
    ClashObjectiveManager : clash_objective_manager = clash_objective_manager{}

    @editable
    ClashEnemiesManager : clash_enemies_manager = clash_enemies_manager{}

    @editable
    ClashCurrencyRewardsManager : clash_currency_rewards_manager = clash_currency_rewards_manager{}

    @editable
    PlayerSpawners : []player_spawner_device = array{}

    @editable
    StartingGold:float=10.0

    #Pasive Income
    @editable
    var PassiveIncomeDelay : float = 10.0

    @editable
    CurrencyWidget:?ui_hud_currency=false

    @editable
    ClashLobbyManager : clash_lobby_manager_device = clash_lobby_manager_device{}

    @editable
    bDebugPrints : logic = false

    var HasMatchStarted : logic = false

    CurrencyChangedEvent:event(currency_changed_payload) = event(currency_changed_payload){}

    #Debug
    GetCurrencyInfo<localizes>(Player : player, Added : float, Total : float) : message = "{Added} added to: {Player}, Total: {Total}"

    # Runs when the device is started in a running game
    OnBegin<override>()<suspends>:void=

        GetPlayspace().PlayerAddedEvent().Subscribe(OnPlayerAdded)

        #Initialize Players
        AllStartingPlayers := GetPlayspace().GetPlayers()
        for (Player : AllStartingPlayers):
            OnPlayerAdded(Player)

        spawn{EventsInitializer()}

    #-----------------------------------------------------------------------------------------------------------------------

    #Initialize player maps
    OnPlayerAdded(Player:player): void=
        if (PlayerExist := PlayersMap[Player]):
        else:
            if (bDebugPrints?):
                Print("Player Added")
            option:
                set PlayersMap[Player] = StartingGold
            if (FortCharacter := Player.GetFortCharacter[]):
                FortCharacter.EliminatedEvent().Subscribe(OnPlayerEliminated)

    #-----------------------------------------------------------------------------------------------------------------------
    #Initialize Events
    EventsInitializer()<suspends>: void=
        sync:
            loop: #Bed Destroyed
                EventInfo := ClashObjectiveManager.OnBedDestroyedEvent.Await()
                OnBedDestroyed(EventInfo)

            loop: #Room Cleared

                #TO DO JJ: Reemplazar este evento ya que se va a elimnar en un futuro
                EventInfo := ClashEnemiesManager.OnStageEmptiedByTeam.Await()
                OnRoomCleared(EventInfo)

            loop: #Npc eliminated
                EventInfo := ClashEnemiesManager.EnemyEliminatedEvent.Await()
                OnNPCEliminated(EventInfo)
            
            loop:
                if(not ClashLobbyManager.MatchStarted?):
                    set HasMatchStarted = ClashLobbyManager.OnMatchGameplayStartEvent.Await()
                    if (HasMatchStarted?):
                        Players := GetPlayspace().GetPlayers()
                        for(Player : Players):
                            AddWidgetToPlayer(Player)
                        PassiveIncome()
                else:
                    PassiveIncome()
    
    #-----------------------------------------------------------------------------------------------------------------------
    PassiveIncome()<suspends> : void =
        loop: #Passive income
            Sleep(PassiveIncomeDelay)
            AllPlayers := GetPlayspace().GetPlayers()

            for (Player : AllPlayers):
                OnPassiveIncome(Player)

    #-----------------------------------------------------------------------------------------------------------------------
    OnPlayerEliminated(EliminResult : elimination_result): void=
        if:
            Killer := EliminResult.EliminatingCharacter?
            Killed := EliminResult.EliminatedCharacter
            KillerAgent := Killer.GetAgent[]
            KillerPlayer := player[KillerAgent]
        then:
            if (not Killer = Killed):
                #Add currency to killer
                TotalReward := ClashCurrencyRewardsManager.GetRewardFromIdForPlayer(KillerPlayer, currency_id.PlayerKill)
                AddCurrency(TotalReward, KillerPlayer, currency_id.PlayerKill)

                CurrencyMessage := GetCurrencyInfo(KillerPlayer, TotalReward, GetCurrency(KillerPlayer))
                if (bDebugPrints?):
                    Print(CurrencyMessage)

                KillerTeam := GetAgentTeam(KillerPlayer)
                if (TeamMembers := GetAgentsInTeam[KillerTeam]):
                    for (AgentMember : TeamMembers):
                        if:
                            PlayerMember := player[AgentMember]
                            PlayerMember <> KillerPlayer
                        then:
                            AssistReward := ClashCurrencyRewardsManager.GetRewardFromIdForPlayer(PlayerMember, currency_id.PlayerAssist)
                            AddCurrency(AssistReward, PlayerMember, currency_id.PlayerAssist)

    #-----------------------------------------------------------------------------------------------------------------------
    OnBedDestroyed(EventInfo : tuple(team_id, int)): void=
        ScoringTeam := EventInfo(0)
        LosingTeam := GetEnemyTeam(ScoringTeam)
    
        if (PlayersScoringTeam := GetAgentsInTeam[ScoringTeam]):
            for (Agent : PlayersScoringTeam):
                if (Player := player[Agent]):
                    #Add currency to scoring team
                    TotalReward := ClashCurrencyRewardsManager.GetRewardFromIdForPlayer(Player, currency_id.ObjectiveDestroyed)
                    AddCurrency(TotalReward, Player, currency_id.ObjectiveDestroyed)

                    CurrencyMessage := GetCurrencyInfo(Player, TotalReward, GetCurrency(Player))
                    if (bDebugPrints?):
                        Print(CurrencyMessage)

        if (PlayersLosingTeam := GetAgentsInTeam[LosingTeam]):
            for (Agent : PlayersLosingTeam):
                if (Player := player[Agent]):

                    #Add currency to losing team
                    TotalReward := ClashCurrencyRewardsManager.GetRewardFromIdForPlayer(Player, currency_id.Underdog)
                    AddCurrency(TotalReward, Player, currency_id.Underdog)

                    CurrencyMessage := GetCurrencyInfo(Player, TotalReward, GetCurrency(Player))
                    if (bDebugPrints?):
                        Print(CurrencyMessage)

    #-----------------------------------------------------------------------------------------------------------------------
    OnPassiveIncome(Player : player): void=
        TotalReward := ClashCurrencyRewardsManager.GetRewardFromIdForPlayer(Player, currency_id.PassiveIncome)
        AddCurrency(TotalReward, Player, currency_id.PassiveIncome)

        CurrencyMessage := GetCurrencyInfo(Player, TotalReward, GetCurrency(Player))
        if (bDebugPrints?):
            Print(CurrencyMessage)

    #-----------------------------------------------------------------------------------------------------------------------
    OnNPCEliminated(EventInfo : stage_interaction_result_payload): void=
        StageIndex := EventInfo.StageIndex
        Interaction := EventInfo.Interaction            
        Spawner := EventInfo.Spawner
        EnemyLevel := ClashEnemiesManager.GetSpawnerEnemyLevelIndex(Spawner)
        CurrencyID := GetEnemyCurrencyIdByLevel(EnemyLevel)
        AssistID := GetEnemyAssistIdByLevel(EnemyLevel)

        if (Agent := Interaction.Source?, Player := player[Agent]):
            TotalReward := ClashCurrencyRewardsManager.GetRewardFromIdForPlayer(Player, CurrencyID)
            AddCurrency(TotalReward, Player, CurrencyID)

            CurrencyMessage := GetCurrencyInfo(Player, TotalReward, GetCurrency(Player))
            if (bDebugPrints?):
                Print("Enemy level: {EnemyLevel}, CurrencyID: {CurrencyID}")
                Print(CurrencyMessage)

            KillerTeam := GetAgentTeam(Player)
            if (TeamMembers := GetAgentsInTeam[KillerTeam]):
                for (AgentMember : TeamMembers):
                    if:
                        PlayerMember := player[AgentMember]
                        PlayerMember <> Player
                    then:
                        AssistReward := ClashCurrencyRewardsManager.GetRewardFromIdForPlayer(PlayerMember, AssistID)
                        AddCurrency(AssistReward, PlayerMember, AssistID)

    #-----------------------------------------------------------------------------------------------------------------------
    OnRoomCleared(EventInfo : tuple(team_id, int))<suspends>: void=
        Team := EventInfo(0)
        if (ScoringTeam := GetAgentsInTeam[Team]):
            for (Agent : ScoringTeam):
                if (Player := player[Agent]):
                    #Add currency to room cleared team
                    TotalReward := ClashCurrencyRewardsManager.GetRewardFromIdForPlayer(Player, currency_id.RoomCapture)
                    AddCurrency(TotalReward, Player, currency_id.RoomCapture)

                    CurrencyMessage := GetCurrencyInfo(Player, TotalReward, GetCurrency(Player))
                    if (bDebugPrints?):
                        Print(CurrencyMessage)
        return

    #-----------------------------------------------------------------------------------------------------------------------
    AddCurrency(Amount:float, Player:player, CurrencySource:currency_id): void=
        if (PlayerCurrency : float = PlayersMap[Player]):
            TotalCurrency := PlayerCurrency + Amount
            
            CurrencyChangedEvent.Signal(
                currency_changed_payload:
                    Player := Player
                    PreviousValue := PlayerCurrency
                    CurrentValue := TotalCurrency
                    CurrencySourceIcon := CurrencySource
                )

            option:
                set PlayersMap[Player] = TotalCurrency
            
            var Added:int=0

            if(Amount < 0.0):
                set Added = 1

            NotifyChangeToWidget(Player, TotalCurrency, Added)

    #-----------------------------------------------------------------------------------------------------------------------
    GetCurrency(Player : player): float=
        if (PlayerCurrency : float = PlayersMap[Player]):
            return PlayerCurrency
        else:
            return 0.0

    #-----------------------------------------------------------------------------------------------------------------------
    AddWidgetToPlayer(Player:player):void=
        #Sleep(Delay)
        if(ValidWidget:=CurrencyWidget?):
            if (bDebugPrints?):
                Print("Currency Widget is Valid")            
            ValidWidget.AddWidgetToPlayer(Player, StartingGold)

    #-----------------------------------------------------------------------------------------------------------------------
    NotifyChangeToWidget(Player:player, Currency:float, Added:int):void=
        if(ValidWidget:=CurrencyWidget?):
            ValidWidget.CurrencyHandlerEvent.Signal(Player, Currency, Added)
