
using { /Fortnite.com/Devices }
using { /Verse.org/Simulation }
using { /UnrealEngine.com/Temporary/Diagnostics }

player_levels_info := struct():
    Player:player

    PlayerLevels:[currency_id]int
    
#-----------------------------------------------------------------------------------------------------------------------
reward_level_data := struct<concrete>():
    @editable
    Level:int = 0

    @editable
    CurrencyValue:float = 0.0

#-----------------------------------------------------------------------------------------------------------------------
currency_data := struct<concrete>():
    @editable
    Id : currency_id = currency_id.None

    @editable
    CurrencyValue : float = 0.0

#-----------------------------------------------------------------------------------------------------------------------
upgrade_currency_data := struct<concrete>():
    @editable
    Id : currency_id = currency_id.None

    @editable
    RewardPerLevel : []reward_level_data = array{reward_level_data{Level := 1, CurrencyValue := 1.0}}

#-----------------------------------------------------------------------------------------------------------------------
clash_currency_rewards_manager := class(creative_device):

    @editable
    CurrencyRewards:[]currency_data = array:
        currency_data{Id := currency_id.Type1EnemyKill, CurrencyValue := 50.0}
        currency_data{Id := currency_id.Type2EnemyKill, CurrencyValue := 50.0}
        currency_data{Id := currency_id.Type3EnemyKill, CurrencyValue := 50.0}
        currency_data{Id := currency_id.Type4EnemyKill, CurrencyValue := 50.0}
        currency_data{Id := currency_id.Type5EnemyKill, CurrencyValue := 50.0}
        currency_data{Id := currency_id.PlayerKill, CurrencyValue := 50.0}
        currency_data{Id := currency_id.Type1EnemyAssist, CurrencyValue := 50.0}
        currency_data{Id := currency_id.Type2EnemyAssist, CurrencyValue := 50.0}
        currency_data{Id := currency_id.Type3EnemyAssist, CurrencyValue := 50.0}
        currency_data{Id := currency_id.Type4EnemyAssist, CurrencyValue := 50.0}
        currency_data{Id := currency_id.Type5EnemyAssist, CurrencyValue := 50.0}
        currency_data{Id := currency_id.PlayerAssist, CurrencyValue := 50.0}
        currency_data{Id := currency_id.CoinReward, CurrencyValue := 50.0}
        currency_data{Id := currency_id.ObjectiveDestroyed, CurrencyValue := 100.0}
        currency_data{Id := currency_id.RoomCapture, CurrencyValue := 50.0}
        currency_data{Id := currency_id.Underdog, CurrencyValue := 50.0}
        currency_data{Id := currency_id.PassiveIncome, CurrencyValue := 1.0}

    @editable
    CurrencyUpgradeValue:[]upgrade_currency_data = array:
        upgrade_currency_data{Id := currency_id.Type1EnemyKill, RewardPerLevel := array{}}
        upgrade_currency_data{Id := currency_id.Type2EnemyKill, RewardPerLevel := array{}}
        upgrade_currency_data{Id := currency_id.Type3EnemyKill, RewardPerLevel := array{}}
        upgrade_currency_data{Id := currency_id.Type4EnemyKill, RewardPerLevel := array{}}
        upgrade_currency_data{Id := currency_id.Type5EnemyKill, RewardPerLevel := array{}}
        upgrade_currency_data{Id := currency_id.PlayerKill, RewardPerLevel := array{}}
        upgrade_currency_data{Id := currency_id.Type1EnemyAssist, RewardPerLevel := array{}}
        upgrade_currency_data{Id := currency_id.Type2EnemyAssist, RewardPerLevel := array{}}
        upgrade_currency_data{Id := currency_id.Type3EnemyAssist, RewardPerLevel := array{}}
        upgrade_currency_data{Id := currency_id.Type4EnemyAssist, RewardPerLevel := array{}}
        upgrade_currency_data{Id := currency_id.Type5EnemyAssist, RewardPerLevel := array{}}
        upgrade_currency_data{Id := currency_id.PlayerAssist, RewardPerLevel := array{}}
        upgrade_currency_data{Id := currency_id.CoinReward, RewardPerLevel := array{}}
        upgrade_currency_data{Id := currency_id.ObjectiveDestroyed, RewardPerLevel := array{}}
        upgrade_currency_data{Id := currency_id.RoomCapture, RewardPerLevel := array{}}
        upgrade_currency_data{Id := currency_id.Underdog, RewardPerLevel := array{}}
        upgrade_currency_data{Id := currency_id.PassiveIncome, RewardPerLevel := array{}}

    var PlayerLevelData:[]player_levels_info = array{}

    #-----------------------------------------------------------------------------------------------------------------------
    GetRewardFromIdForPlayer(Player:player, CurrencyID:currency_id) : float =
        var BasePlayerReward : float = 0.0
        var ExtraRewardForLevel : float = 0.0

        LevelToCheck := GetPlayerLevelForCurrency(Player, CurrencyID)

        # Base Value for reward
        for (CurrentReward : CurrencyRewards):
            if (CurrencyID = CurrentReward.Id):
                set BasePlayerReward = CurrentReward.CurrencyValue

        # Additional value for every level
        for (CurrentUpgradeValue : CurrencyUpgradeValue):
            if (CurrencyID = CurrentUpgradeValue.Id):
                for (CurrentLevel : CurrentUpgradeValue.RewardPerLevel):
                    if (CurrentLevel.Level = LevelToCheck):
                        set ExtraRewardForLevel = CurrentLevel.CurrencyValue

        return BasePlayerReward + ExtraRewardForLevel
        
    #-----------------------------------------------------------------------------------------------------------------------
    GetPlayerLevelForCurrency(PlayerToCheck:player, ID:currency_id) : int =
        for(CurrentData : PlayerLevelData):
            if (CurrentData.Player = PlayerToCheck):
                for(CurrentIndex -> CurrencyLevel : CurrentData.PlayerLevels):
                    if (CurrentIndex = ID):
                        return CurrencyLevel
            else:
                set PlayerLevelData += array{}
        return 0