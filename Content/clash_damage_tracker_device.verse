
using { /Fortnite.com/Devices }
using { /Verse.org/Simulation }
using { /UnrealEngine.com/Temporary/Diagnostics }
using { /Fortnite.com/Characters }
using { /Fortnite.com/Game }

clash_damage_tracker_device := class(creative_device):

    OnPlayerDamagedEvent:event(damage_result) = event(damage_result){}
    OnPlayerKilledEvent:event(elimination_result) = event(elimination_result){}

    #---------------------------------------------------------------------------
    OnBegin<override>()<suspends>: void =
        AllPlayers: []player = GetPlayspace().GetPlayers()
        for (Player : AllPlayers):
            SubscribePlayerToEvents(Player)

        GetPlayspace().PlayerAddedEvent().Subscribe(SubscribePlayerToEvents)

    #---------------------------------------------------------------------------
    SubscribePlayerToEvents(Player : player): void =
        if:
            FortChar := Player.GetFortCharacter[]
        then:
            FortChar.DamagedEvent().Subscribe(DamageEvent)
            FortChar.EliminatedEvent().Subscribe(EliminatedEvent)

    #---------------------------------------------------------------------------
    DamageEvent(Result : damage_result) : void =
        OnPlayerDamagedEvent.Signal(Result)

    EliminatedEvent(Result : elimination_result): void =
        OnPlayerKilledEvent.Signal(Result)