
using { /Fortnite.com/Devices }
using { /Verse.org/Simulation }
using { /UnrealEngine.com/Temporary/Diagnostics }
using { /Verse.org/Simulation/Tags }
using { /UnrealEngine.com/Temporary/SpatialMath }

debug_device_tag:=class(tag){}

clash_debug_helper_device := class(creative_device):

    @editable
    bAllowDebugDevices : logic = true
    @editable
    TeleportLocation : vector3 = vector3{}

    SearchingTag:tag = debug_device_tag{}

    OnBegin<override>()<suspends>:void =
        if(bAllowDebugDevices = false):
            GetAndHideAllDebugDevices()

    GetAndHideAllDebugDevices(): void =
        CreativeObjects := FindCreativeObjectsWithTag(SearchingTag)

        for (CreativeObject:CreativeObjects):
            if:
                ValidDevice := creative_device[CreativeObject]
                ValidDevice.TeleportTo[TeleportLocation, IdentityRotation()]
            then:
                Print("All debug devices are hided")
