
using { /Fortnite.com/Devices }
using { /Verse.org/Simulation }
using { /UnrealEngine.com/Temporary/Diagnostics }

clash_defensive_structure_activator_device := class(creative_device):

    #Debug
    #-----------------------------------------------------------------------------------------------------------------------------------------

    Logger:log=log{Channel:=log_defensive_structure_manager, DefaultLevel:=log_level.Normal}

    @editable
    bDebugEnabled:logic = false

    #Variables
    #-----------------------------------------------------------------------------------------------------------------------------------------

    @editable
    ButtonDevice: button_device = button_device{}
    @editable
    AssociatedDefensiveStructure: clash_defensive_structure_device = clash_defensive_structure_device{}
    @editable
    MessageData: defensive_structure_message_data = defensive_structure_message_data{}

    var CurrentOwnerTeamID: team_id = team_id.Neutral

    #TEST ONLY
    @editable
    ButtonTeamSwitch: button_device = button_device{}

    #-----------------------------------------------------------------------------------------------------------------------------------------
    OnBegin<override>()<suspends>:void =
        ButtonDevice.InteractedWithEvent.Subscribe(ActivateStructure)
        ButtonTeamSwitch.InteractedWithEvent.Subscribe(SwitchOwnershipByAgent)

    #Use this function to switch the owning team of the device by an agent
    #-----------------------------------------------------------------------------------------------------------------------------------------
    SwitchOwnershipByAgent(Agent: agent): void =
        NewTeam := GetAgentTeam(Agent)
        if(CurrentOwnerTeamID <> NewTeam):
            set CurrentOwnerTeamID = NewTeam
            MessageData.SendOwnershipMessage(Agent, AssociatedDefensiveStructure.DefensiveStructureDisplayName)

            if(bDebugEnabled?):
                Logger.Print("Switched device owner to team: {CurrentOwnerTeamID}")

    #-----------------------------------------------------------------------------------------------------------------------------------------
    ActivateStructure(Agent : agent): void =
        if:
            ActivatingAgentTeam := GetAgentTeam(Agent)
            Player := player[Agent]
            CurrentOwnerTeamID = ActivatingAgentTeam
            AssociatedDefensiveStructure.bIsDeviceActivated = false
        then:
            #TODO: GET PLAYERS MONEY AND COMPARE HERE WITH DEVICE PRICE
            if(AssociatedDefensiveStructure.Price <= 10.0):
                #TODO: TAKE MONEY FROM PLAYER
                AssociatedDefensiveStructure.CreateDefensiveStructureForPlayer(Player)
                ButtonDevice.Disable()
                MessageData.SendCreatedStructureMessage(Agent, AssociatedDefensiveStructure.DefensiveStructureDisplayName, 0.0)
                spawn{HandleButtonLoop()}
            else:
                MessageData.SendNotEnoughMoneyMessage(Agent)

                if(bDebugEnabled?):
                    Logger.Print("Not enough money")
        else:
            MessageData.SendNotAllowedInteractionMessage(Agent, false)

            if(bDebugEnabled?):
                Logger.Print("User not owner")

    #-----------------------------------------------------------------------------------------------------------------------------------------
    HandleButtonLoop()<suspends>: void =

        TerminateButtonHandlerEvent:event()=event(){}

        race:
            TerminateButtonHandlerEvent.Await()
            loop:
                AssociatedDefensiveStructure.OnDefensiveStructureDeactivatedEvent.Await()
                ButtonDevice.Enable()
                TerminateButtonHandlerEvent.Signal()