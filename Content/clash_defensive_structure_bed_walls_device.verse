
using { /Fortnite.com/Devices }
using { /Verse.org/Simulation }
using { /UnrealEngine.com/Temporary/Diagnostics }
using { /Fortnite.com/Game }
using { /Fortnite.com/Characters }

clash_defensive_structure_bed_walls_device := class(clash_defensive_structure_device):

    DefensiveStructureDisplayName<override>:string = "Bed Wall"
    DefensiveStructureDescription<override>:string = ""
    DefensiveStructureID<override>:defensive_structure_id = defensive_structure_id.BedDefenses
    bUsesPreviewModel<override>:logic = true

    @editable
    ObjectiveDeviceCollection: []objective_device = array{}
    @editable
    VFXDevice: vfx_spawner_device = vfx_spawner_device{}

    var LastLayer : objective_device = objective_device{}

    #-----------------------------------------------------------------------------------------------------------------------------------------
    OnBegin<override>()<suspends>:void =
        (super:)OnBegin()
        ToggleLayers(false)

        #objective_device Damage Event is bugged, leave these lines to fix it
        Sleep(2.0)
        for(ObjectiveDevice : ObjectiveDeviceCollection):
            ObjectiveDevice.Damage(5.0)
            ObjectiveDevice.Heal(5.0)
        #----------------------------------------------------------

        #Subscribe all layers to deactivation event
        for(ObjectiveDevice : ObjectiveDeviceCollection):
            ObjectiveDevice.DamagedEvent().Subscribe(OnLayerDamaged)

    #-----------------------------------------------------------------------------------------------------------------------------------------
    ActivateDefensiveStructure<override>(Player : player): void =
        (super:)ActivateDefensiveStructure(Player)
        ToggleLayers(true)

    #-----------------------------------------------------------------------------------------------------------------------------------------
    DeactivateDefensiveStructure<override>(): void =
        (super:)DeactivateDefensiveStructure()

    #-----------------------------------------------------------------------------------------------------------------------------------------
    UpgradeDefensiveStructure<override>(Player : player): void =
        (super:)UpgradeDefensiveStructure(Player)

    #-----------------------------------------------------------------------------------------------------------------------------------------
    ToggleLayers(bOn : logic): void =
        for (ObjectiveDevice : ObjectiveDeviceCollection):
            if(bOn?):
                ObjectiveDevice.Show()
            else:
                ObjectiveDevice.Hide()

    #-----------------------------------------------------------------------------------------------------------------------------------------
    OnLayerDamaged(Result : damage_result): void =
        var Counter: int = 0

        for (ObjectiveDevice : ObjectiveDeviceCollection):
            if:
                ObjectiveDevice.GetHealth() <= 0.0
            then:
                ObjectiveDevice.Hide()
                ObjectiveDevice.Heal(ObjectiveDevice.GetMaxHealth())
                VFXDevice.Restart()
                if:
                    Counter = 0
                    Instigator := Result.Instigator?
                    Agent := Instigator.GetInstigatorAgent[]
                then:
                    OnDefensiveStructureDestroyed(option{Agent})
            set Counter += 1