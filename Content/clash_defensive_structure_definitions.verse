using { /Verse.org/Simulation }
using { /Verse.org/Simulation/Tags }
using { /UnrealEngine.com/Temporary/Diagnostics }
using { /UnrealEngine.com/Temporary/UI }
using { /Fortnite.com/UI }
using { /Fortnite.com/Devices }

#Tags
#-------------------------------------------------------------------------------------------------
defensive_structure_device_tag:=class(tag){}

#Enums
#-------------------------------------------------------------------------------------------------

defensive_structure_id := enum:
    None,
    AutomatedTurret,
    DefensiveWall,
    ImprovedGuards,
    BedDefenses

defensive_structure_id_array: []defensive_structure_id = array{
    defensive_structure_id.AutomatedTurret
    defensive_structure_id.DefensiveWall
    defensive_structure_id.ImprovedGuards
    defensive_structure_id.BedDefenses
}

#Debug
#-------------------------------------------------------------------------------------------------

log_defensive_structure_manager := class(log_channel){}

#Messages
#-------------------------------------------------------------------------------------------------

defensive_structure_message_data := class<concrete>():
    
    @editable
    HUDMessageDevice:hud_message_device = hud_message_device{}

    #Content
    #-------------------------------------------------------------------------------------------------

    @editable #EX: Created: Automated Turret For: 3 Golds
    DefensiveStructureBoughtTextA: string = "Created: "
    @editable 
    DefensiveStructureBoughtTextB: string = " For: "
    @editable 
    NotTheOwnerOfDeviceText: string = "You are not the owner of this device"
    @editable 
    NotAllowedIntercationToDevice: string = "You are not allowed to interact with this device"
    @editable 
    NotEnoughMoneyText: string = "Not enough money"
    @editable
    OwnershipText: string = "You are now the owner of: "

    #-------------------------------------------------------------------------------------------------
    SendCreatedStructureMessage(Agent: agent, PartA: string, PartB: float): void =
        HUDMessageDevice.SetText(StringToMessage(DefensiveStructureBoughtTextA + PartA + DefensiveStructureBoughtTextB + "{PartB}"))
        HUDMessageDevice.Show(Agent)

    #-------------------------------------------------------------------------------------------------
    SendNotAllowedInteractionMessage(Agent: agent, bIsOwner: logic): void =
        if (bIsOwner?):
            HUDMessageDevice.SetText(StringToMessage(NotAllowedIntercationToDevice))
        else:
            HUDMessageDevice.SetText(StringToMessage(NotTheOwnerOfDeviceText))
        HUDMessageDevice.Show(Agent)

    #-------------------------------------------------------------------------------------------------
    SendNotEnoughMoneyMessage(Agent: agent): void =
        HUDMessageDevice.SetText(StringToMessage(NotEnoughMoneyText))
        HUDMessageDevice.Show(Agent)

    #-------------------------------------------------------------------------------------------------
    SendOwnershipMessage(Agent: agent, PartA: string): void =
        HUDMessageDevice.SetText(StringToMessage(OwnershipText + "{PartA}"))
        HUDMessageDevice.Show(Agent)