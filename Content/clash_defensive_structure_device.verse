using { /Fortnite.com/Devices }
using { /Fortnite.com/Characters }
using { /Fortnite.com/UI }
using { /Verse.org/Simulation }
using { /Verse.org/Assets }
using { /Verse.org/Concurrency }
using { /Verse.org/Colors }
using { /Verse.org/Colors/NamedColors }
using { /UnrealEngine.com/Temporary/Diagnostics }
using { /UnrealEngine.com/Temporary/UI }
using { /UnrealEngine.com/Temporary/SpatialMath }

clash_defensive_structure_device := class(creative_device):

    Logger:log=log{Channel:=log_defensive_structure_manager, DefaultLevel:=log_level.Normal}

    @editable
    bDebugEnabled:logic = false

    #Variables
    #-----------------------------------------------------------------------------------------------------------------------------------------

    @editable
    DefensiveStructureDisplayName:string = "Defensive Structure Name"
    @editable
    DefensiveStructureDescription:string = "Defensive Structure Description"
    @editable
    Price:float = 0.0
    @editable
    var bIsDeviceActivated:logic = false
    @editable
    bUsesPreviewModel:logic = false
    @editable
    PreviewModel:creative_prop = creative_prop{}
    
    DefensiveStructureID : defensive_structure_id = defensive_structure_id.None

    #Events
    #-----------------------------------------------------------------------------------------------------------------------------------------

    OnDefensiveStructureCreatedEvent:event() = event() {}
    OnDefensiveStructureDeactivatedEvent:event() = event() {}
    OnDefensiveStructureUpgradedEvent:event() = event() {}

    #-----------------------------------------------------------------------------------------------------------------------------------------
    OnBegin<override>()<suspends>:void =
        if(bUsesPreviewModel?):
            PreviewModel.Show()
        else:
            PreviewModel.Hide()

    #-----------------------------------------------------------------------------------------------------------------------------------------
    GetDefensiveStructureName(): string = return DefensiveStructureDisplayName

    #-----------------------------------------------------------------------------------------------------------------------------------------
    GetDefensiveStructureDescription(): string = return DefensiveStructureDescription
    
    #-----------------------------------------------------------------------------------------------------------------------------------------
    CreateDefensiveStructureForPlayer(Player:player): void =
        ActivateDefensiveStructure(Player)
        OnDefensiveStructureCreatedEvent.Signal()
        spawn{DefensiveStructureHandler(Player)}

        if(bDebugEnabled?):
            Logger.Print("Creating Defensive Structure: {DefensiveStructureDisplayName}")

    #Handles system deactivation when conditions are met
    #-----------------------------------------------------------------------------------------------------------------------------------------
    DefensiveStructureHandler<protected>(Player:player)<suspends>:void =
        
        TerminateDefensiveStructureHandlerEvent:event()=event(){}

        race:
            TerminateDefensiveStructureHandlerEvent.Await()
            sync:
                loop: #If player dies, remove if necessary
                    if(FortCharacter:=Player.GetFortCharacter[]):
                        FortCharacter.EliminatedEvent().Await()
                        OnDeactivateDevice(Player, TerminateDefensiveStructureHandlerEvent)

                loop: #if defensive structure is deactivated from other source
                    Data := OnDefensiveStructureDeactivatedEvent.Await()
                    if(Data = Player):
                        OnDeactivateDevice(Player, TerminateDefensiveStructureHandlerEvent)

    #-----------------------------------------------------------------------------------------------------------------------------------------
    OnDeactivateDevice<protected>(Player:player, TerminateEvent:event()): void =
        TerminateEvent.Signal()
        DeactivateDefensiveStructure()
        
    #Call this function when conditions to destroy a defensive structure device are met
    #-----------------------------------------------------------------------------------------------------------------------------------------
    OnDefensiveStructureDestroyed<protected>(MaybeAgent:?agent):void =
        if:
            Agent := MaybeAgent?
        then:
            DeactivateDefensiveStructure()

    #Override these three methods to control each defensive device functionality
    #-----------------------------------------------------------------------------------------------------------------------------------------
    ActivateDefensiveStructure(Player : player): void = 
        set bIsDeviceActivated = true
        if(bUsesPreviewModel?):
            PreviewModel.Hide()

            if(bDebugEnabled?):
                Logger.Print("Hiding preview model for {DefensiveStructureDisplayName}")

        if(bDebugEnabled?):
            Logger.Print("Parent method called: ActivateDefensiveStructure()")

    #-----------------------------------------------------------------------------------------------------------------------------------------
    DeactivateDefensiveStructure(): void =
        OnDefensiveStructureDeactivatedEvent.Signal()
        set bIsDeviceActivated = false
        if(bUsesPreviewModel?):
            PreviewModel.Show()

            if(bDebugEnabled?):
                Logger.Print("Displaying preview model for {DefensiveStructureDisplayName}")

        if(bDebugEnabled?):
            Logger.Print("Parent method called: DeactivateDefensiveStructure()")

    #-----------------------------------------------------------------------------------------------------------------------------------------
    UpgradeDefensiveStructure(Player: player): void = 
        OnDefensiveStructureUpgradedEvent.Signal()

        if(bDebugEnabled?):
            Logger.Print("Parent method called: UpdgradeDefensiveStructure()")