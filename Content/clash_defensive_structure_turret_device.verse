
using { /Fortnite.com/Devices }
using { /Verse.org/Simulation }
using { /UnrealEngine.com/Temporary/Diagnostics }
using { /UnrealEngine.com/Temporary/SpatialMath }


clash_defensive_structure_turret_device := class(clash_defensive_structure_device):

    DefensiveStructureDisplayName<override>:string = "Automated Turret"
    DefensiveStructureDescription<override>:string = ""
    DefensiveStructureID<override>:defensive_structure_id = defensive_structure_id.AutomatedTurret
    bUsesPreviewModel<override>:logic = false

    #Use one turret for each team since switching teams in runtine is bugged for turret devices
    @editable
    TeamBlueTurretDevice:automated_turret_device = automated_turret_device{}
    @editable
    TeamRedTurretDevice:automated_turret_device = automated_turret_device{}

    var OrgPos:vector3 = vector3{}

    #-----------------------------------------------------------------------------------------------------------------------------------------
    OnBegin<override>()<suspends>:void =

        TeamBlueTurretDevice.DestroyedEvent.Subscribe(OnDefensiveStructureDestroyed)
        TeamRedTurretDevice.DestroyedEvent.Subscribe(OnDefensiveStructureDestroyed)

        set OrgPos = TeamBlueTurretDevice.GetTransform().Translation

        (super:)OnBegin()

    #-----------------------------------------------------------------------------------------------------------------------------------------
    ActivateDefensiveStructure<override>(Player : player): void =

        CurrentTeam := GetAgentTeam(Player)

        case (CurrentTeam):
            team_id.Blue =>
                if:
                    TeamBlueTurretDevice.TeleportTo[OrgPos, TeamBlueTurretDevice.GetTransform().Rotation]
                    TeamRedTurretDevice.TeleportTo[OrgPos + vector3{Z := -5.0}, TeamRedTurretDevice.GetTransform().Rotation]
                then:
                    TeamBlueTurretDevice.Enable()
            team_id.Red =>
                if:
                    TeamRedTurretDevice.TeleportTo[OrgPos, TeamRedTurretDevice.GetTransform().Rotation]
                    TeamBlueTurretDevice.TeleportTo[OrgPos + vector3{Z := -5.0}, TeamBlueTurretDevice.GetTransform().Rotation]
                then:
                    TeamRedTurretDevice.Enable()
            _=>

        (super:)ActivateDefensiveStructure(Player)

    #-----------------------------------------------------------------------------------------------------------------------------------------
    DeactivateDefensiveStructure<override>(): void =

        TeamBlueTurretDevice.Disable()
        TeamRedTurretDevice.Disable()

        (super:)DeactivateDefensiveStructure()

    #-----------------------------------------------------------------------------------------------------------------------------------------
    UpgradeDefensiveStructure<override>(Player : player): void =
        (super:)UpgradeDefensiveStructure(Player)

    #-----------------------------------------------------------------------------------------------------------------------------------------