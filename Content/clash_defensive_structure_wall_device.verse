
using { /Fortnite.com/Devices }
using { /Fortnite.com/Game }
using { /Fortnite.com/Characters }
using { /Verse.org/Simulation }
using { /UnrealEngine.com/Temporary/SpatialMath }
using { /UnrealEngine.com/Temporary/Diagnostics }

clash_defensive_structure_wall_device := class(clash_defensive_structure_device):

    DefensiveStructureDisplayName<override>:string = "Defensive Wall"
    DefensiveStructureDescription<override>:string = ""
    DefensiveStructureID<override>:defensive_structure_id = defensive_structure_id.DefensiveWall
    bUsesPreviewModel<override>:logic = true

    @editable
    ObjectiveDevice: objective_device = objective_device{}
    @editable
    VFXDevice: vfx_spawner_device = vfx_spawner_device{}

    #-----------------------------------------------------------------------------------------------------------------------------------------
    OnBegin<override>()<suspends>:void =
        (super:)OnBegin()
        ObjectiveDevice.Hide()

        #objective_device Damage Event is bugged, leave these three lines to fix it
        Sleep(2.0)
        ObjectiveDevice.Damage(10.0)
        ObjectiveDevice.Heal(10.0)
        #----------------------------------------------------------

        ObjectiveDevice.DamagedEvent().Subscribe(OnWallDamaged)
        
    #-----------------------------------------------------------------------------------------------------------------------------------------
    ActivateDefensiveStructure<override>(Player : player): void =
        (super:)ActivateDefensiveStructure(Player)
        SpawnDefensiveWall()

    #-----------------------------------------------------------------------------------------------------------------------------------------
    DeactivateDefensiveStructure<override>(): void =
        (super:)DeactivateDefensiveStructure()

    #-----------------------------------------------------------------------------------------------------------------------------------------
    UpgradeDefensiveStructure<override>(Player : player): void =
        (super:)UpgradeDefensiveStructure(Player)

    #-----------------------------------------------------------------------------------------------------------------------------------------
    SpawnDefensiveWall(): void =
        ObjectiveDevice.Show()

    #-----------------------------------------------------------------------------------------------------------------------------------------
    OnWallDamaged(Result: damage_result): void =
        if:
            ObjectiveDevice.GetHealth() <= 0.0
        then:
            ObjectiveDevice.Hide()
            ObjectiveDevice.Heal(ObjectiveDevice.GetMaxHealth())

            if:
                Instigator := Result.Instigator?
                Agent := Instigator.GetInstigatorAgent[]
            then:
    
                VFXDevice.Restart()
                OnDefensiveStructureDestroyed(option{Agent})