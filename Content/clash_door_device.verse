
using { /Fortnite.com/Devices }
using { /Verse.org/Simulation }
using { /UnrealEngine.com/Temporary/Diagnostics }

log_clash_door_device := class(log_channel){} # TODO JJ: Study how rooms work

clash_door_device := class(creative_device):
    Logger:log = log{Channel := log_clash_door_device}
    
    @editable
    DisplayName<public>:string = "0"

    @editable
    Barrier:barrier_device = barrier_device{}

    @editable
    bDebug:logic = false

    var AllowedTeams:[]team_id = array{}

    # Runs when the device is started in a running game
    OnBegin<override>()<suspends>:void=
        if (bDebug?) then Logger.Print("Initialization of door {DisplayName}")

    AddAllowedPlayer(Agent:agent):void =
        Barrier.AddToIgnoreList(Agent)

    ResetPlayersAllowed():void =
        Barrier.RemoveAllFromIgnoreList()
    
    RemoveAllowedPlayer(Agent:agent):void =
        Barrier.RemoveFromIgnoreList(Agent)
    
    TemporarilyBlockAll():void = 
        Barrier.RemoveAllFromIgnoreList()

    AddAllowedTeam(Team:team_id):void =
        if (Agents := GetAgentsInTeam[Team]):
            for (Agent:Agents, Player := player[Agent]):
                AddAllowedPlayer(Player)
        if (bDebug?):
            Logger.Print("Added allowed team {Team} in {DisplayName}")
        
    RemoveAllowedTeam(Team:team_id):void =
        if (Agents := GetAgentsInTeam[Team]):
            for (Agent:Agents, Player := player[Agent]):
                RemoveAllowedPlayer(Player)
        if (bDebug?):
            Logger.Print("Removed allowed team {Team} in {DisplayName}")

    
