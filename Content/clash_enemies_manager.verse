using { /Fortnite.com/Devices }
using { /Fortnite.com/Characters }
using { /Fortnite.com/Playspaces }
using { /Fortnite.com/Game }
using { /Verse.org/Simulation }
using { /Verse.org/Simulation/Tags }
using { /UnrealEngine.com/Temporary/Diagnostics }



log_clash_enemies_manager := class(log_channel){}

# Tags to identify Spawners by Stage (each stage is one of the playable rooms). Tags are useful to handle a bunch of references to devices without adding their referneces one per one (more info here https://www.youtube.com/watch?v=wXMr6IKhwIo)
ClashGuardSpawner_S1_Tag := class(tag){}
ClashGuardSpawner_S2_Tag := class(tag){}
ClashGuardSpawner_S3_Tag := class(tag){}
ClashGuardSpawner_S4_Tag := class(tag){}
ClashGuardSpawner_S5_Tag := class(tag){}
ClashGuardSpawner_S6_Tag := class(tag){}
ClashGuardSpawner_S7_Tag := class(tag){}
ClashGuardSpawner_S8_Tag := class(tag){}
ClashGuardSpawner_S9_Tag := class(tag){}
ClashGuardSpawner_S10_Tag := class(tag){}
ClashGuardSpawner_S11_Tag := class(tag){}
ClashGuardSpawner_S12_Tag := class(tag){}
ClashGuardSpawner_S13_Tag := class(tag){}
ClashGuardSpawner_S14_Tag := class(tag){}
ClashGuardSpawner_S15_Tag := class(tag){}

# Enemy Level
ClashEnemy_Level_1_Tag := class(tag){} 
ClashEnemy_Level_2_Tag := class(tag){}
ClashEnemy_Level_3_Tag := class(tag){}
ClashEnemy_Level_4_Tag := class(tag){}   
ClashEnemy_Level_5_Tag := class(tag){} 

# Bosses Level
ClashBoss_Level_1_Tag := class(tag){} 
ClashBoss_Level_2_Tag := class(tag){} 
ClashBoss_Level_3_Tag := class(tag){} 
ClashBoss_Level_4_Tag := class(tag){} 
ClashBoss_Level_5_Tag := class(tag){} 

GetClashGuardSpawnerTags():[]tag = {
	ClashGuardSpawner_S1_Tag{},
	ClashGuardSpawner_S2_Tag{},
	ClashGuardSpawner_S3_Tag{},
	ClashGuardSpawner_S4_Tag{},
	ClashGuardSpawner_S5_Tag{},
	ClashGuardSpawner_S6_Tag{},
	ClashGuardSpawner_S7_Tag{},
	ClashGuardSpawner_S8_Tag{},
	ClashGuardSpawner_S9_Tag{},
	ClashGuardSpawner_S10_Tag{},
	ClashGuardSpawner_S11_Tag{},
	ClashGuardSpawner_S12_Tag{},
	ClashGuardSpawner_S13_Tag{},
	ClashGuardSpawner_S14_Tag{},
	ClashGuardSpawner_S15_Tag{}
}

GetClashEnemyLevelTags():[]tag ={
    ClashEnemy_Level_1_Tag{}, 
    ClashEnemy_Level_2_Tag{},
    ClashEnemy_Level_3_Tag{},
    ClashEnemy_Level_4_Tag{},   
    ClashEnemy_Level_5_Tag{}
    #ClashBoss_Level_1_Tag{},
    #ClashBoss_Level_2_Tag{},
    #ClashBoss_Level_3_Tag{},
    #ClashBoss_Level_4_Tag{},
    #ClashBoss_Level_5_Tag{}
}

ClashEmpty_Tag := class(tag){}

clash_stage := class(creative_device):

	Logger:log = log{Channel:=log_clash_enemies_manager}

	# Guard spawners for this stage
	var Spawners:[]guard_spawner_device = array{}

	# Keeps reference of the agents spawned and alive in this stage
	var SpawnedAgents:[]agent = array{}

	# Know if the stage is despawning so eliminated events are not sent
	var bIsDespawning:logic = false

	# Integrated Team Definitions 
	var CurrentTeam:team_id = team_id.Neutral

	# Enemies killed for team 1
	var KillsTeam1:int = 0

	# Enemies killed for team 2
	var KillsTeam2:int = 0

	#@editable
	#var EnemyManagerReference: clash_enemies_manager = clash_enemies_manager{}

	# For handling giving room to team that made the most kills. Reports kill for a team in this stage.
	ReportKill(TeamID:team_id): void =
		case(TeamID):
			team_id.Blue =>
				set KillsTeam1 += 1
			team_id.Red =>
				set KillsTeam2 += 1
			_ =>
				Logger.Print("Neutral kill")

	ClearKills(): void =
		set KillsTeam1 = 0
		set KillsTeam2 = 0

	# For handling giving room to team that made the most kills. Reports team currently winning in the stage.
	TeamWinning(): team_id =
		if (KillsTeam1>KillsTeam2):
			return team_id.Blue

		if (KillsTeam2>KillsTeam1):	  
			return team_id.Red
			  
		return team_id.Neutral

	# For handling giving room to team that made the most kills. Returns ratio of kills, from 0 to 1
	GetPlayerKillsRatio(TeamID:team_id)<transacts>: int =

		var MaxCountguards: int= GetMaxGuardsCount()

		if (MaxCountguards <= 0):
			return 0

		case(TeamID):
			team_id.Blue =>
				if (Percent := ((KillsTeam1*100)/MaxCountguards)):
					return Floor(Percent)
			team_id.Red =>
				if (Percent := ((KillsTeam2*100)/MaxCountguards)):
					return Floor(Percent)
			_ =>
				return 0
		
		return 0

	AddSpawnedAgent(Agent:agent): void =
		set SpawnedAgents += array{Agent}

	RemoveSpawnedAgent(Agent:agent): void =
		set SpawnedAgents = SpawnedAgents.RemoveAllElements(Agent)

	ResetSpawnedAgents<protected>(): void =
		set SpawnedAgents = array{}

	GetLivingGuardsCount()<transacts>: int =
		SpawnedAgents.Length

	# Returns the max count of guards this stage will have
	GetMaxGuardsCount()<transacts>:int =
		var MaxGuardsCount:int = 0
		for (Spawner : Spawners):
			set MaxGuardsCount += Spawner.GetSpawnLimit()
		MaxGuardsCount

	# Returns if one or guard is dead and current living guards is not the max one. If the room is full will return false.
	AreSpawnersDirty()<decides><transacts>:void =
		GetLivingGuardsCount() < GetMaxGuardsCount()

	# Función para manejar el evento de generación de un guardia # posiblemente deprecar !
	HandleGuardSpawned(newguard: agent):void=
		# Asignar el equipo al guardia generado
		var MaxGuardsCount:int = 0

		#newguard.EliminatedEvent().Subscribe(OnGuardReceivedDamage)

		AssignToTeamByIndex(newguard, CurrentTeam)

		#newguard.ChangeTeam(1)
		
		#self.context.GetPlayspace().GetTeamCollection().AddToTeam[newguard, 1]
		
		#if:
		#	Teams := GetPlayspace().GetTeamCollection() #creates a collection of teams in the game currently.
		#	Team := Teams.GetTeam[newguard] #Gets the current team number of the instigating player.
		#	TeamIndex := Teams.GetTeams().Find[Team] #Converts that number to an int.
		#	TeamIndex = 0 #Compares the player's team number to the one you are looking for (0 is Team 1 in this case)
		#then:
			#Do Stuff based on Team 1

		#print("Guardia generado y asignado al equipo:", guard.GetTeam())

	# WORKS !
	AssignToTeamByIndex(Agent: agent, TeamID:team_id):void=
		TeamCollection := Self.GetPlayspace().GetTeamCollection()
		Teams := TeamCollection.GetTeams()
		if (
			#Team := Teams[Index-1]
			Team := Teams[GetTeamIndex(TeamID)]
			TeamCollection.AddToTeam[Agent, Team]
		)
			then:
				Print("New Team assigned {ToString(TeamID)}")
			else:
				Print("Could not assign team")

	# Resets spawners if dirty
	InitializeSpawnersIfDirty():void = {}
		<#if (AreSpawnersDirty[]): # Thanos: Maybe ?
			Print("***** Room Manager: Room Manager: Initializing spawners NPCs  *****")
			DespawnAll()

			for (Spawner : Spawners):
				Spawner.SpawnedEvent.Subscribe(HandleGuardSpawned)
				for (Count := 0..(Spawner.GetSpawnLimit()-1)):
					Spawner.Spawn()#>

	# Resets spawners if the room is clear of NPCs
	InitializeSpawnersIfClear():void = {}
		<#if (GetLivingGuardsCount() = 0): # Thanos: Maybe too ?
			Print("***** Room Manager: Empty room, Initializing spawners NPCs *****")
			DespawnAll()

			for (Spawner : Spawners):
				Spawner.SpawnedEvent.Subscribe(HandleGuardSpawned)
				for (Count := 0..(Spawner.GetSpawnLimit()-1)):
					Spawner.Spawn()#>

	# Resets spawners by force
	InitializeSpawnersByForce():void = {}
		<#Logger.Print("***** Room Manager: Spawning by force!  *****") # Thanos: Yep, Was Here
		DespawnAll()

		for (Spawner : Spawners):
			Spawner.SpawnedEvent.Subscribe(HandleGuardSpawned)
			for (Count := 0..(Spawner.GetSpawnLimit()-1)):
				Spawner.Spawn()#>


	DisableSpawners():void =
		DespawnAll()

	DespawnAll<protected>(): void = 
		set bIsDespawning = true

		for (Spawner : Spawners):
			Spawner.Despawn()
		
		ResetSpawnedAgents()

		set bIsDespawning = false
	

# A Verse-authored creative device that can be placed in a level
clash_enemies_manager := class(creative_device):

	Logger:log = log{Channel:=log_clash_enemies_manager}

	# The stages of our clash game
	var GameStages: []clash_stage = array{}

	# TODO JJ events that are pending
	OnStageEmptyEvent<public>:event(tuple(agent, int)) = event(tuple(agent, int)){}

	# TODO JJ events that are pending
	OnStageEmptiedByTeam<public>:event(tuple(team_id, int)) = event(tuple(team_id, int)){}

	OnStageNotEmptyEvent<public>:event(int) = event(int){}

	EnemyReceivedDamageEvent:event(damage_result) = event(damage_result){}

	EnemyEliminatedEvent:event(stage_interaction_result_payload) = event(stage_interaction_result_payload){}

	@editable
	TimeBeforeSpawning:float = 3.0

	# Runs when the device is started in a running game
	OnBegin<override>()<suspends>:void=

		# Lets get our enemies using the tags !
		var DebugStageIndex:int=0
		GuardSpawnerTags:=GetClashGuardSpawnerTags()
		for (GuardSpawnerTag : GuardSpawnerTags):
			set DebugStageIndex += 1
			NewStage := clash_stage{Spawners := GetWorldSpawners(GuardSpawnerTag) <#, Team1Device := Team1Device, Team2Device := Team2Device,Team3Device := Team3Device , EnemyManagerReference := Self #>}
			
			set GameStages += array{NewStage}
			Logger.Print("####  Stage: {DebugStageIndex}, Spawners Found: {NewStage.Spawners.Length}")
			#PrintStageSpawnersPosition(DebugStageIndex-1)

		Logger.Print("Total Stages: {GameStages.Length}")

		# Bind Enemy Spawn and Dead
		for (GameStageIndex := 0..(GameStages.Length-1)):
			for (Spawner: GameStages[GameStageIndex].Spawners):
				spawn{AwaitGuardSpawnersEvents(GameStageIndex, Spawner)}

	# Función para manejar el evento de generación de un guardia
	#OnGuardSpawnedTest(newguard: agent):void=
		# Asignar el equipo al guardia generado
		#var MaxGuardsCount:int = 0
		#newguard.ChangeTeam(1)

		#MyTeam:team := team(1)
		
		#if (GetPlayspace().GetTeamCollection().AddToTeam[newguard, 1]):
		#	Logger.Print("success!")

	GetWorldSpawners(Tag: tag) : []guard_spawner_device =
		TaggedActors := FindCreativeObjectsWithTag(Tag)
			var Objects:[]creative_object_interface=array{}
				for (Obj:TaggedActors, TaggedActor:=creative_device_base[Obj]):
					set Objects += array{TaggedActor}
					#return Objects
		var FoundSpawners:[]guard_spawner_device = array{}
		
		set FoundSpawners = for:
			TaggedActor : TaggedActors
			SpawnerDevice := guard_spawner_device[TaggedActor]
		do:
			SpawnerDevice
		return FoundSpawners

	GetSpawnedAgentsInStage(Index : int) : []agent =
		SpawnedAgents : []agent = array{}
		if (SpawnedAgents = GameStages[Index].SpawnedAgents) {}
			return SpawnedAgents

	ActivateStage(Index:int) : void =
		if (Index <= INDEX_NONE):
			return

		if (GameStage := GameStages[Index-1]):
			Logger.Print("#### Try Initializing spawners NPCs in stage {Index} room {Index}")
			GameStage.InitializeSpawnersIfClear() # JJ: Temp, not respawning if there are alive guards # InitializeSpawnersIfDirty

			<#
		# WORKS silly test, force team change in every npc already spawned:
		for (GameStageIndex := 0..(GameStages.Length-1)):
			if (GameStage := GameStages[GameStageIndex]):
				for (Agent : GameStage.SpawnedAgents):
					AssignToTeamByIndex(Agent, 2) # Seems to work !
					#>

	# JJ: check if need to deprecate
	ActivateStageForTeam(Index:int, TeamID:team_id) : void =
		Logger.Print("ActivateStageForTeam Index: {Index} Team: {ToString(TeamID)}")
		
		case(TeamID):
			team_id.Blue =>
				if (GameStage := GameStages[Index+1]):
					set GameStage.CurrentTeam= TeamID
					GameStage.InitializeSpawnersByForce()
			team_id.Red =>
				if (GameStage := GameStages[Index-1]):
					set GameStage.CurrentTeam= TeamID
					GameStage.InitializeSpawnersByForce()
			_ =>
				if (GameStage := GameStages[Index]):
					set GameStage.CurrentTeam= TeamID
					GameStage.InitializeSpawnersByForce()
				Logger.Print("#### Defaulted to {ToString(TeamID)}")

		Logger.Print("#### Try Initializing spawners NPCs in stage {Index} room {Index} in team {ToString(TeamID)}")
		

	DeactivateStage(Index:int) : void =
		if (Index <= INDEX_NONE):
			return

		if (GameStage := GameStages[Index-1]):
			Logger.Print("Disabling spawners NPCs in stage {Index} room {Index}")
			GameStage.DisableSpawners()

	# Here We wait for events such as guards receiving damage, dying and spawning. Have ane eye for this
	AwaitGuardSpawnersEvents(StageIndex:int, Spawner:guard_spawner_device)<suspends>:void =
		sync:
			loop:
				# Print("Enemy Spawned!")
				Agent:agent = Spawner.SpawnedEvent.Await()

				<# TODO JJ : Enemy receiving damage.
				#Cbray: Implemented @DamagedShieldEvent, Jan/24/24.
				Remove this event if you want to notify health damage only, take into account BoosterSystem uses this event to 
				detect enemy was damaged when they have shield#>
				if (FortCharacter:=Agent.GetFortCharacter[]):
					# Print("Enemy FC Cast Correct!")
					FortCharacter.DamagedEvent().Subscribe(OnGuardReceivedDamage)
					FortCharacter.DamagedShieldEvent().Subscribe(OnGuardReceivedDamage)

				if (GameStage:= GameStages[StageIndex]):
					LastLivingGuardsCount := GameStage.GetLivingGuardsCount()

					GameStage.AddSpawnedAgent(Agent)

					if (LastLivingGuardsCount = 0):
						OnStageNotEmptyEvent.Signal(StageIndex)
						
					Logger.Print("#### Stage: {StageIndex+1}, Living Guard Added, Total Living Guards: {GameStage.GetLivingGuardsCount()}")

			loop:
				Result:device_ai_interaction_result = Spawner.EliminatedEvent.Await()

				EnemyEliminatedEvent.Signal(stage_interaction_result_payload{Interaction:=Result, StageIndex:=StageIndex, Spawner := Spawner})

				if:
					GameStage:= GameStages[StageIndex]
					not GameStage.bIsDespawning?
					KilledAgent:=Result.Target?
					GameStage.SpawnedAgents.Find[KilledAgent] # If not added means it was also called after a despawning
					MurdererAgent:= Result.Source?
				then:
					GameStage.RemoveSpawnedAgent(KilledAgent)
					
					MurdererTeam :team_id= GetAgentTeam(MurdererAgent)

					# Lets report a kill by the given team
					GameStage.ReportKill(MurdererTeam)

					if (GameStage.GetLivingGuardsCount() <= 0):
						Logger.Print("<<<<<<<<<<<< Empty room ! {StageIndex}")

						OnStageEmptyEvent.Signal(MurdererAgent, StageIndex + 1)
						OnStageEmptiedByTeam.Signal(GameStage.TeamWinning(), StageIndex + 1)
						# stage is clear !, lets respawn it for the other team and see what happens
						
						TeamCollection := GetPlayspace().GetTeamCollection()
						if (AgentsTeam := TeamCollection.GetTeam[MurdererAgent]):
							TeamArray := TeamCollection.GetTeams()
							for (TeamNumber->Team : TeamArray):
								if (AgentsTeam = Team):
									Logger.Print("<<<<<<<<<<<< Team of murderer ? {TeamNumber+1}")
									case(TeamNumber):
										0 => # Blue Team
											Sleep(TimeBeforeSpawning)
											ActivateStageForTeam(StageIndex-1, team_id.Blue)
										1 => # Red Team
											Sleep(TimeBeforeSpawning)
											ActivateStageForTeam(StageIndex+1, team_id.Red)
										_ =>
											Logger.Print("Something seems wrong, Team ? {TeamNumber+1}")
						
						GameStage.ClearKills()

						
						

					# TODO JJ: Rewards are pending
					#if (PlayerKiller := player[Result.Source?]):
					#	EnemyLevel := GetSpawnerEnemyLevelIndex(Spawner)
					#	if (EnemyLevel >= 0):
					#		GiveEliminationCurrencyReward(PlayerKiller, EnemyLevel)
							
					Logger.Print("#### Stage: {StageIndex+1}, Left Guards: {GameStage.GetLivingGuardsCount()}")


	OnGuardReceivedDamage(Result:damage_result): void =
		Print("Enemy Damaged Event Binded!")
		EnemyReceivedDamageEvent.Signal(Result)

	# WORKS This one seems to work !, please always use teams counting form one onwards
	<#
	AssignToTeamByIndex(Agent: agent, Index:int):void=
		TeamCollection := Self.GetPlayspace().GetTeamCollection()
		Teams := TeamCollection.GetTeams()
		if (
			Team := Teams[Index-1]
			TeamCollection.AddToTeam[Agent, Team]
		)
			then:
				Print("New Team assigned {Index-1}")
			else:
				Print("Could not assign team")
	#>

	<#
	# Testing WORKS deprecate
	HandleGuardSpawnedFromDevice(newguard: agent):void=
		# Asignar el equipo al guardia generado
		var MaxGuardsCount:int = 0
		AssignToTeamByIndex(newguard, 2)
		#>

	GetSpawnerEnemyLevelIndex(Spawner:guard_spawner_device):int =
		SpawnerTags := Spawner.GetTags()
		EnemyLevelTags:=GetClashEnemyLevelTags()
		for (TagIndex := 0..(EnemyLevelTags.Length-1), Tag := EnemyLevelTags[TagIndex]):
			if (SpawnerTags.Has[Tag]):
				return TagIndex + 1
	
		return INDEX_NONE

	#-----------------------------------------------------------------------------------------------------------------------
	ResetStageForScoringTeam(ScoringTeamID:team_id):void=
		for (StageIndex -> CurrentGameStage:GameStages):
			if(CurrentGameStage.CurrentTeam = ScoringTeamID):
				ActivateStageForTeam(StageIndex, team_id.Neutral)
