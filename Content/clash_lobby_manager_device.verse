
using { /Fortnite.com/Devices }
using { /Fortnite.com/Characters }
using { /Fortnite.com/FortPlayerUtilities }
using { /Verse.org/Simulation }
using { /UnrealEngine.com/Temporary/Diagnostics }
using { /Verse.org/Random }
using { /Verse.org/Simulation/Tags }

lobby_player_spawner_device_tag:=class(tag){}

lobby_manager_device_tag:=class(tag){}

clash_lobby_manager_device := class(creative_device):


	@editable
	MaxMembersInRedTeam : int = 2
	@editable
	MaxMembersInBlueTeam : int = 2
	@editable
	StartingTimer : timer_device = timer_device{}
	@editable
	RedTeleporter : teleporter_device = teleporter_device{}
	@editable
	BlueTeleporter : teleporter_device = teleporter_device{}
	@editable
	NeutralTeleporter : teleporter_device = teleporter_device{}

	@editable
	NeutralTeamPlayerReferenceDevice : class_and_team_selector_device = class_and_team_selector_device{}
	@editable
	BlueTeamPlayerReferenceDevice : class_and_team_selector_device = class_and_team_selector_device{}
	@editable
	RedTeamPlayerReferenceDevice : class_and_team_selector_device = class_and_team_selector_device{}

	#-----Team selection-----
	@editable
	RedVolume : volume_device = volume_device{}
	@editable
	BlueVolume : volume_device = volume_device{}

	@editable
	RedTrigger : trigger_device = trigger_device{}
	@editable
	BlueTrigger : trigger_device = trigger_device{}

	@editable
	CosmeticBlueTeamPlayerReferenceDevice : class_and_team_selector_device = class_and_team_selector_device{}
	@editable
	CosmeticRedTeamPlayerReferenceDevice : class_and_team_selector_device = class_and_team_selector_device{}
	@editable
	PlayerCounters : []player_counter_device = array{}

	#-----Lobby spawners-----
	@editable
	SpawnerTag:tag = player_spawner_device_tag{}
	@editable
	LobbySpawnerTag:tag = lobby_player_spawner_device_tag{}

	#-----Events-----
	# Event for indicating if the match gameplay portion has ended or starrted. Useful for stuff like hidding or showing the HUD
	OnMatchGameplayStartEvent<public>:event(logic) = event(logic){}
	@editable
	EventDisplayerDevice : clash_ui_event_displayer_device = clash_ui_event_displayer_device{}
	@editable
	RespawnTime : float = 10.0
	
	#-----Audio-----   
	@editable
	AudioManagerDevice: clash_audio_manager = clash_audio_manager{}
	@editable
	LobbyMusicDelay : float = 8.0
	var MatchStarted : logic = false
	var AllowTeamSelection : logic = false

	#-----Debug-----
	var IsLobbyEnabled : logic = true

	@editable 
	BeginMatch_DebugButton: button_device = button_device{}

	@editable 
	BeginMatchWithWait_DebugButton: button_device = button_device{}

	#--- TEST TO FORBID MATCHMAKING
	@editable 
	NoMatchmake_DebugButton: button_device = button_device{}

	@editable
	RoundHandlerDevice :round_settings_device = round_settings_device{}


	# Optional board to debug in private codes
	@editable
	var DebugBoard : billboard_device = billboard_device{}

	var DebugTextString : string = ""

	#StringyToMessage<localizes><public>(String:string)<computes>:message = "{String}"
	DebugBoardMessage<localizes>(String:string)<computes> : message = " Sup! {String}"


	OnBegin<override>()<suspends>:void =
		if:
			IsLobbyEnabled?
		then:
			DisablePlayerSpawners()
			# Test for fourth player being stuck in lobby issue
			GetPlayspace().PlayerAddedEvent().Subscribe(CheckForMatchStart)
			Sleep(1.0)

			UpdateLobbyTeamSelectors(false)

		for (Player : GetPlayspace().GetPlayers()):
			CheckForMatchStart(Player)
		#Sub incoming players to neutral team		
		
		
		#RedTrigger.TriggeredEvent.Subscribe(AssignToRedTeam)
		#BlueTrigger.TriggeredEvent.Subscribe(AssignToBlueTeam)

		RedVolume.AgentEntersEvent.Subscribe(AssignToRedTeam)
		BlueVolume.AgentEntersEvent.Subscribe(AssignToBlueTeam)

		StartingTimer.SuccessEvent.Subscribe(StartGameForPlayers)
		BeginMatch_DebugButton.InteractedWithEvent.Subscribe(MatchStartDebug)
		BeginMatchWithWait_DebugButton.InteractedWithEvent.Subscribe(MatchStartDelayedDebug)

		NoMatchmake_DebugButton.InteractedWithEvent.Subscribe(DisableMatchmaking)
		

	CheckForMatchStart(Agent : agent): void =
		if (MatchStarted = true):
			#TODO Define logic when match started
			AssignNeutralToTeam()
			BillboardDebugUpdate("CheckForMatchStart: Match already started")
		else: #
			BillboardDebugUpdate("CheckForMatchStart: Match NOT started")
			#NeutralTeamPlayerReferenceDevice.ChangeTeam(Agent)
			spawn:
				TestMakeneutral(Agent)

			Players := GetPlayspace().GetPlayers()
			#Check if enough players for Match Start
			if (Players.Length=(MaxMembersInBlueTeam+MaxMembersInRedTeam)):
				UpdateLobbyTeamSelectors(true)		
				EventDisplayerDevice.DisplayChooseTeamMessage(Agent)
				StartingTimer.Start()
				#OnMatchGameplayStartEvent.Signal(true)
				spawn:
					LaunchDelayedGameStartSignal(true)
			else:
				spawn:
					PlayLobbyMusicAfterDelay(Agent)
		
	
	PlayLobbyMusicAfterDelay(Agent : agent)<suspends> : void =
		Sleep(LobbyMusicDelay)
		AudioManagerDevice.PlayAudio_ElevatorMusic(Agent,false)

	# Are the lobby team selectors enabled or not ?
	UpdateLobbyTeamSelectors(Enable : logic):void =
		if:
			Enable?
		then:
			set AllowTeamSelection = true
			#RedVolume.Enable()
			#BlueVolume.Enable()
			#RedTrigger.Enable()
			#BlueTrigger.Enable()
			CosmeticBlueTeamPlayerReferenceDevice.Enable()
			CosmeticRedTeamPlayerReferenceDevice.Enable()
			for (PlayerCounter : PlayerCounters):
				PlayerCounter.ShowInfoPanel()
		else:
			set AllowTeamSelection = false
			#RedVolume.Disable()
			#BlueVolume.Disable()
			#RedTrigger.Disable()
			#BlueTrigger.Disable()
			CosmeticBlueTeamPlayerReferenceDevice.Disable()
			CosmeticRedTeamPlayerReferenceDevice.Disable()
			for (PlayerCounter : PlayerCounters, PlayerCounterDevice := player_counter_device[PlayerCounter]):
				PlayerCounterDevice.HideInfoPanel()
	
	EnablePlayerSpawners():void=
		DisableLobbyPlayerSpawners()
		CreativeObjects:=FindCreativeObjectsWithTag(SpawnerTag)
		for(Creativedevice: CreativeObjects, Spawner := player_spawner_device[Creativedevice]):
		   Spawner.Enable()
	
	DisablePlayerSpawners():void=
		CreativeObjects:=FindCreativeObjectsWithTag(SpawnerTag)
		for(Creativedevice: CreativeObjects, Spawner := player_spawner_device[Creativedevice]):
		   Spawner.Disable()

	EnableLobbyPlayerSpawners():void=
		CreativeObjects:=FindCreativeObjectsWithTag(LobbySpawnerTag)
		for(Creativedevice: CreativeObjects, Spawner := player_spawner_device[Creativedevice]):
			Spawner.Enable()
		
	DisableLobbyPlayerSpawners():void=
		CreativeObjects:=FindCreativeObjectsWithTag(LobbySpawnerTag)
		for(Creativedevice: CreativeObjects, Spawner := player_spawner_device[Creativedevice]):
			Spawner.Disable()
	
	AssignToRedTeam(MaybeAgent : agent): void =
		if:
			not AllowTeamSelection?
		then:
			return

		var RedAgents : int = 0
		if:
			Length :=  GetAgentsInTeam[team_id.Red].Length
			set RedAgents = Length

		BillboardDebugUpdate("Red: How Many: {RedAgents} < the Max {MaxMembersInBlueTeam}")

		if:
			Agent := MaybeAgent
			RedAgents < MaxMembersInRedTeam
		then:
			AudioManagerDevice.PlayAudio_TeamSelectedRed(Agent)
			RedTeamPlayerReferenceDevice.ChangeTeam(Agent)
			EventDisplayerDevice.DisplayRedTeamHUD(Agent)
			#return true
		#return false

	AssignToBlueTeam(MaybeAgent : agent): void =
		if:
			not AllowTeamSelection?
		then:
			return

		var BlueAgents : int = 0
		if:
			Length :=  GetAgentsInTeam[team_id.Blue].Length
			set BlueAgents = Length

		BillboardDebugUpdate("Blue: How Many: {BlueAgents} < the Max {MaxMembersInBlueTeam}")

		if:
			Agent := MaybeAgent
			BlueAgents < MaxMembersInBlueTeam
				
		then:
			AudioManagerDevice.PlayAudio_TeamSelectedBlue(Agent)
			BlueTeamPlayerReferenceDevice.ChangeTeam(Agent)
			EventDisplayerDevice.DisplayBlueTeamHUD(Agent)
			#return true
		#return false

	AssignNeutralToTeam() : void =
		if:
			NeutralAgents:= GetAgentsInTeam[team_id.Neutral].Length > 0
		then:
			for(NeutralAgent: GetAgentsInTeam[team_id.Neutral] ):
				AssignToBlueSuccess:= AssignToBlueTeam(NeutralAgent) #option{NeutralAgent}
				if (AssignToBlueSuccess=false):
					AssignToRedTeam(NeutralAgent) #option{NeutralAgent}
	
	MatchStartDebug(Agent : agent): void =  
		StartGameForPlayers(option{Agent})
		OnMatchGameplayStartEvent.Signal(true)

	MatchStartDelayedDebug(Agent : agent): void =
		UpdateLobbyTeamSelectors(true)	
		EventDisplayerDevice.DisplayChooseTeamMessage(Agent)
		StartingTimer.Start()
		#OnMatchGameplayStartEvent.Signal(true)
		spawn:
			LaunchDelayedGameStartSignal(true)

	DisableMatchmaking(Agent : agent): void =  
		RoundHandlerDevice.DisableMatchmaking()


	StartGameForPlayers(MaybeAgent : ?agent):void =
		if(Agent := MaybeAgent?):
			AudioManagerDevice.PlayAudio_ElevatorMusic(Agent,true)
		AudioManagerDevice.PlayAudio_MatchStart()
		set MatchStarted = true		
		EnablePlayerSpawners()
		#Asign players that are missing a team
		AssignNeutralToTeam()
		
		#Respawn Players to start
		for (BluePlayer : GetAgentsInTeam[team_id.Blue]):
			if:
				FortCharacter := BluePlayer.GetFortCharacter[]
			then:
				FortCharacter.Damage(FortCharacter.GetMaxHealth())

		for (RedPlayer : GetAgentsInTeam[team_id.Red]):
			if:
				FortCharacter := RedPlayer.GetFortCharacter[]
			then:
				FortCharacter.Damage(FortCharacter.GetMaxHealth())

		# signal was here
				
	BillboardDebugUpdate(DebugText: string): void =
		Print(DebugText)
		set DebugTextString += DebugText + "\n"
		DebugBoard.SetText(DebugBoardMessage(DebugTextString))

	
	LaunchDelayedGameStartSignal( IsGameStarting: logic)<suspends>:void =
		TimeToStart := StartingTimer.GetMaxDuration()
		BillboardDebugUpdate("Time: {TimeToStart}")
		Sleep(TimeToStart)
		OnMatchGameplayStartEvent.Signal(true)
		Sleep(RespawnTime)
		Players := GetPlayspace().GetPlayers()
		for(Player : Players):			
			EventDisplayerDevice.DisplayMatchStartMessage(Player)
	
	# JUST FOR TESTING
	TestMakeneutral(Agent : agent)<suspends>:void =
		Sleep(0.1)
		BillboardDebugUpdate("MAKING NEUTRAL AT START !")
		NeutralTeamPlayerReferenceDevice.ChangeTeam(Agent)
		#NeutralTeleporter.Teleport(Agent)

