
using { /Fortnite.com/Devices }
using { /Verse.org/Simulation }
using { /UnrealEngine.com/Temporary/Diagnostics }

clash_match_end_manager_device := class(creative_device):
    @editable
    LobbyTeleport : teleporter_device = teleporter_device{}

    @editable
    VictoryScreenCamera : gameplay_camera_fixed_point_device = gameplay_camera_fixed_point_device{}
    
    @editable
    WinnerTeamReferenceDevices : []player_reference_device = array{}

    @editable
    WinnerShowcaseDuration : float = 5.0

    @editable
    RematchTimerDevice : timer_device = timer_device{}

    @editable
    RoundSettingsDevice : round_settings_device = round_settings_device{}

    @editable
    ItemRemover : item_remover_device = item_remover_device{}

    var MVPPlayer : ?agent = false

    #-----------------------------------------------------------------------------------------------------------------------
    OnBegin<override>()<suspends>:void=
        RematchTimerDevice.SuccessEvent.Subscribe(StartRematch)

    #-----------------------------------------------------------------------------------------------------------------------
    StartRoundEnd(WinnerTeam : team_id) <suspends> : void =
        RematchTimerDevice.Start()
        if(WinnerAgents := GetAgentsInTeam[WinnerTeam]):
            set MVPPlayer = option{WinnerAgents[0]}
            for (PlayerIndex -> CurrentAgent : WinnerAgents):
                if:
                    CurrentPlayer := player[CurrentAgent]
                    DeviceToEnable := WinnerTeamReferenceDevices[PlayerIndex]
                then:
                    DeviceToEnable.Register(CurrentPlayer)

        PlayersAtTheEnd := GetPlayspace().GetPlayers()

        for (CurrentPlayer : PlayersAtTheEnd):
            VictoryScreenCamera.AddTo(CurrentPlayer)
            LobbyTeleport.Teleport(CurrentPlayer)
            ItemRemover.Remove(CurrentPlayer)

        Sleep(WinnerShowcaseDuration)

        for (CurrentPlayer : PlayersAtTheEnd):
            VictoryScreenCamera.RemoveFrom(CurrentPlayer)

    #-----------------------------------------------------------------------------------------------------------------------
    StartRematch(MaybeAgent : ?agent) : void =
        if (WinnerAgent := MVPPlayer?):
            RoundSettingsDevice.EndRound(WinnerAgent)