
using { /Fortnite.com/Devices }
using { /Fortnite.com/Characters }
using { /Verse.org/Simulation }
using { /UnrealEngine.com/Temporary/Diagnostics }
using { /UnrealEngine.com/Temporary/SpatialMath }
using { /Fortnite.com/FortPlayerUtilities }
using { /Fortnite.com/Game }

clash_objective_manager := class(creative_device):

    @editable
    bEnableDebug:logic = false

    #Simple common messages UI Stuff
	@editable
	UIEventMessageDisplayerDevice: clash_ui_event_displayer_device = clash_ui_event_displayer_device{}

    @editable
    BlueBeds:[]objective_device = array{}

    @editable
    RedBeds:[]objective_device = array{}

    @editable
    BlueTeamTeleport:teleporter_device = teleporter_device{}

    @editable
    RedTeamTeleport:teleporter_device = teleporter_device{}

    @editable
    ScoreManagerRef:score_manager_device = score_manager_device{}

    # @editable
    # EndGameRef:end_game_device =end_game_device{}

    @editable
    RoomManager:clash_room_manager = clash_room_manager{}

    @editable
    ScoreToWin:int = 3

    @editable
    EndRoundItemRemover : item_remover_device = item_remover_device{}

    OnBedDestroyedEvent:event(tuple(team_id, int)) = event(tuple(team_id, int)){}

    OnUpdatedVictoryTrackerEvent: event(tuple(team_id, int)) = event(tuple(team_id, int)){}

    OnGameFinished:event(team_id) = event(team_id){}

    var BlueBedsDestroyed:int = 0

    var RedBedsDestroyed:int = 0

    var BlueKills:int = 0

    var RedKills:int = 0

    var BlueBedTransform:transform = transform{}

    var RedBedTransform:transform = transform{}

    var ScoringTeamTransform:transform = transform{}

    #-----BETWEEN ROUNDS CINEMATIC-----
    @editable
    InBetweenRoundDuration : float = 5.0

    @editable
    RedBedCamera : gameplay_camera_fixed_point_device = gameplay_camera_fixed_point_device{}

    @editable
    BlueBedCamera : gameplay_camera_fixed_point_device = gameplay_camera_fixed_point_device{}

    @editable
    BedDestroyedSequenceDevice : cinematic_sequence_device = cinematic_sequence_device{}

    @editable
    RedBedDestroyedVFX : vfx_spawner_device = vfx_spawner_device{}

    @editable
    BlueBedDestroyedVFX : vfx_spawner_device = vfx_spawner_device{}

    @editable
    BoosterManager : booster_manager_device = booster_manager_device{}

    @editable
    AudioManager : clash_audio_manager = clash_audio_manager{}

    #-----END OF GAME-----
    @editable
    MatchEndReference:clash_match_end_manager_device = clash_match_end_manager_device{}
    
    #-----------------------------------------------------------------------------------------------------------------------
    OnBegin<override>()<suspends>:void=
        for (CurrentObjectiveDevice : BlueBeds):
            CurrentObjectiveDevice.Hide()
            CurrentObjectiveDevice.DestroyedEvent.Subscribe(OnObjectiveDestroyed)
            CurrentObjectiveDevice.DamagedEvent().Subscribe(OnObjectiveDamage)
            CurrentObjectiveDevice.Damage(1.0)
            CurrentObjectiveDevice.Heal(1.0)

        for (CurrentObjectiveDevice : RedBeds):
            CurrentObjectiveDevice.Hide()
            CurrentObjectiveDevice.DestroyedEvent.Subscribe(OnObjectiveDestroyed)
            CurrentObjectiveDevice.DamagedEvent().Subscribe(OnObjectiveDamage)
            CurrentObjectiveDevice.Damage(1.0)
            CurrentObjectiveDevice.Heal(1.0)

        if:
            FirstBlueBed := BlueBeds[0]
            FirstRedBed := RedBeds[0]
            set BlueBedTransform = FirstBlueBed.GetTransform()
            set RedBedTransform = FirstRedBed.GetTransform()
        then:
            FirstBlueBed.Show()
            FirstRedBed.Show()
    
    #-----------------------------------------------------------------------------------------------------------------------
    OnObjectiveDamage(Result : damage_result): void =
        if:
            Instigator := Result.Instigator?
            Agent := Instigator.GetInstigatorAgent[]
            InstigatorTeamID := GetAgentTeam(Agent)
            InstigatorTeam := GetAgentsInTeam[InstigatorTeamID]
            RivalTeamID := GetEnemyTeam(InstigatorTeamID)
            RivalTeam := GetAgentsInTeam[RivalTeamID]
            Result.Amount > 0.0
        then:            
            for(AgentInRivalTeam : RivalTeam):
                UIEventMessageDisplayerDevice.SendEnemyBedDamagedMessage(AgentInRivalTeam, RivalTeamID)
            for (AgentInAllyTeam : InstigatorTeam):
                UIEventMessageDisplayerDevice.SendEnemyBedDamagedMessage(AgentInAllyTeam, RivalTeamID)

    #-----------------------------------------------------------------------------------------------------------------------
    OnObjectiveDestroyed(DestroyerAgent:agent):void=
        var BedsDestroyed:int = 0
        var ScoringTeam:team_id = team_id.Neutral
        var EndRoundMessage:string = ""
        if:
            DestroyerTeam := GetAgentTeam(DestroyerAgent)            
            AgentsOnDestroyerTeam := GetAgentsInTeam[DestroyerTeam]
        then:
            RoomManager.RestartAllRooms()
                
            for (Player : AgentsOnDestroyerTeam):
                ScoreManagerRef.Activate(Player)
                CurrentTeam := GetAgentTeam(Player)
                
            case(DestroyerTeam):
                team_id.Blue =>
                    set RedBedsDestroyed += 1
                    set BedsDestroyed = RedBedsDestroyed
                    set ScoringTeam = team_id.Blue
                    RedBedDestroyedVFX.Restart()
                    if(bEnableDebug?):
                        Print("ObjManager: Blue Team Score")
                team_id.Red =>
                    set BlueBedsDestroyed += 1
                    set BedsDestroyed = BlueBedsDestroyed
                    set ScoringTeam = team_id.Red
                    BlueBedDestroyedVFX.Restart()
                    if(bEnableDebug?):
                        Print("ObjManager: Red Team Score")
                _ =>
                    if(bEnableDebug?):
                        Print("ObjManager: Team not valid")

            OnUpdatedVictoryTrackerEvent.Signal(DestroyerTeam, BedsDestroyed)
            UIEventMessageDisplayerDevice.ClearHUDForAllPlayers()
            spawn{ObjectiveDestroyed(BedsDestroyed, DestroyerTeam)}

    #-----------------------------------------------------------------------------------------------------------------------
    ObjectiveDestroyed(ObjectivesDestroyed:int, ScoringTeam:team_id) <suspends> : void =
        var CameraToFocus : gameplay_camera_fixed_point_device = gameplay_camera_fixed_point_device{}
        case (ScoringTeam):
            team_id.Blue =>
                set CameraToFocus = RedBedCamera
            team_id.Red =>
                set CameraToFocus = BlueBedCamera
            _ =>
                if(bEnableDebug?):
                    Print("ObjManager: Invalid Team")

        AllPlayers := GetPlayspace().GetPlayers()
        for (CurrentPlayer : AllPlayers):
            EndRoundItemRemover.Remove(CurrentPlayer)
            CameraToFocus.AddTo(CurrentPlayer)
            if(ScoringTeam = team_id.Blue):
                if(ObjectivesDestroyed >= ScoreToWin):
                    AudioManager.PlayAudio_BlueVictory(CurrentPlayer)
                else:
                    AudioManager.PlayAudio_RedBedDestroyed(CurrentPlayer)
            else:
                if(ObjectivesDestroyed >= ScoreToWin):
                    AudioManager.PlayAudio_RedVictory(CurrentPlayer)
                else:
                    AudioManager.PlayAudio_BlueBedDestroyed(CurrentPlayer)

        BedDestroyedSequenceDevice.Play()
        
        if (ObjectivesDestroyed >= ScoreToWin):
            OnGameFinished.Signal(ScoringTeam)
        else:
            OnBedDestroyedEvent.Signal(ScoringTeam, ObjectivesDestroyed)

        Sleep(InBetweenRoundDuration)

        BedDestroyedSequenceDevice.Stop()

        for (CurrentPlayer : AllPlayers):
            CameraToFocus.RemoveFrom(CurrentPlayer)
            if (CurrentFort := CurrentPlayer.GetFortCharacter[]):
                CurrentFort.SetHealth(CurrentFort.GetMaxHealth())
        
        if (ObjectivesDestroyed >= ScoreToWin):
            CompleteGame(ObjectivesDestroyed, ScoringTeam)
        else:
        EnableNextObjective(ObjectivesDestroyed, ScoringTeam)

    #-----------------------------------------------------------------------------------------------------------------------
    EnableNextObjective(ObjectivesDestroyed:int, ObjectiveBelongingTeam:team_id):void=
        var TeamObjectives:[]objective_device = array{}
        case (ObjectiveBelongingTeam):
            team_id.Blue =>
                set TeamObjectives = RedBeds
                set ScoringTeamTransform = RedBedTransform                
            team_id.Red =>
                set TeamObjectives = BlueBeds
                set ScoringTeamTransform = BlueBedTransform            
            _ =>
                if(bEnableDebug?):
                    Print("ObjManager: Invalid Team")

        if:
            NextBedToEnable := TeamObjectives[ObjectivesDestroyed]
            NextBedToEnable.TeleportTo[ScoringTeamTransform]
        then:
            NextBedToEnable.Show()
            NextBedToEnable.SetInvulnerable(false)

        # Teleport players back to base
        for (Player : GetPlayspace().GetPlayers()):
            CurrentTeam := GetAgentTeam(Player)
            case(CurrentTeam):
                team_id.Blue =>
                    BlueTeamTeleport.Teleport(Player)
                team_id.Red =>
                    RedTeamTeleport.Teleport(Player)
                _ =>
                    if(bEnableDebug?):
                        Print("ObjManager: Team not valid")
            
            BoosterManager.RemoveAllBoostersFromPlayer(Player)

    #-----------------------------------------------------------------------------------------------------------------------
    CompleteGame(ObjectivesDestroyed:int, ScoringTeam:team_id):void =
        spawn{MatchEndReference.StartRoundEnd(ScoringTeam)}
