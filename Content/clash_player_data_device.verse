
using { /Fortnite.com/Devices }
using { /Verse.org/Simulation }
using { /Verse.org/Simulation/Tags }
using { /UnrealEngine.com/Temporary/Diagnostics }

clash_player_data_device_tag:=class(tag){}
#=========================================================================
clash_player_data:=class():
    var BoosterData:[booster_id]int=map{}
    var PowerupData:[powerup_id]int=map{}
#----------------------------------------------------------------------
(PlayerData:clash_player_data).GetBoosterData(BoosterID:booster_id)<decides><transacts>:int=
    return PlayerData.BoosterData[BoosterID]
#----------------------------------------------------------------------
(PlayerData:clash_player_data).GetPowerupData(PowerupID:powerup_id)<decides><transacts>:int=
    return PlayerData.PowerupData[PowerupID]
#----------------------------------------------------------------------
(PlayerData:clash_player_data).SetBoosterData(BoosterID:booster_id, UpgradeLevel:int):void=
    if:
        set PlayerData.BoosterData[BoosterID] = UpgradeLevel
#----------------------------------------------------------------------
(PlayerData:clash_player_data).SetPowerupData(PowerupID:powerup_id, UpgradeLevel:int):void=
    if:
        set PlayerData.PowerupData[PowerupID] = UpgradeLevel
#=========================================================================
clash_player_data_device := class(creative_device):
    var PlayerData:[player]clash_player_data=map{}
#=========================================================================
    OnBegin<override>()<suspends>:void=
        GetPlayspace().PlayerRemovedEvent().Subscribe(OnPlayerRemoved)
#=========================================================================
    AddPlayerData(Player:player):void=
        if(not PlayerData[Player]):
            if:
                set PlayerData[Player]=clash_player_data{}
#=========================================================================
    SetPlayerBoosterUpgradeLevel(Player:player, BoosterID:booster_id, NewUpgradeLevel:int):void=
        #Make Sure this player's data exist
        AddPlayerData(Player)
        if:
            TargetPlayerData:=PlayerData[Player]
        then:
            TargetPlayerData.SetBoosterData(BoosterID, NewUpgradeLevel)
#=========================================================================
    GetPlayerBoosterUpgradeLevel(Player:player, BoosterID:booster_id)<decides><transacts>:int=
        TargetPlayerData:=PlayerData[Player]
        return TargetPlayerData.GetBoosterData[BoosterID]   
#=========================================================================
    SetPlayerPowerupUpgradeLevel(Player:player, PowerupID:powerup_id, NewUpgradeLevel:int):void=
        #Make Sure this player's data exist
        AddPlayerData(Player)
        if:
            TargetPlayerData:=PlayerData[Player]
        then:
            TargetPlayerData.SetPowerupData(PowerupID, NewUpgradeLevel)
#=========================================================================
    GetPlayerPowerupUpgradeLevel(Player:player, PowerupID:powerup_id)<decides><transacts>:int=
        TargetPlayerData:=PlayerData[Player]
        return TargetPlayerData.GetPowerupData[PowerupID]  
#=========================================================================
    RemovePlayerData(Player:player):void=
        var TempMap:[player]clash_player_data=map{}
        for (Key -> Value : PlayerData, Player <> Key):
                set TempMap = ConcatenateMaps (TempMap, map{Key => Value})
        set PlayerData = TempMap
#=========================================================================      
    OnPlayerRemoved(Player:player):void=
        RemovePlayerData(Player)