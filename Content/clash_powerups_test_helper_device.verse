
using { /Fortnite.com/Devices }
using { /Verse.org/Simulation }
using { /UnrealEngine.com/Temporary/Diagnostics }

clash_test_helper_powerups_device := class(test_helper_powerups_device):