using { /Verse.org/Simulation/Tags }
using { /Verse.org/Simulation }
using { /UnrealEngine.com/Temporary/UI }
using { /UnrealEngine.com/Temporary/Diagnostics }
using { /Fortnite.com/UI }

#=======================================================================================================================

powerups_array:[]powerup_id=array{
    powerup_id.MaxHealth,
    powerup_id.MaxShield,
    powerup_id.MaxDamage,
    powerup_id.ShieldRegen,
    powerup_id.HealthRegen
}

IndexToPowerupID(Index:int):powerup_id=
    if:
        ID:= powerups_array[Index]
    then:
        return ID
    return powerup_id.None

