
using { /Fortnite.com/Devices }
using { /Fortnite.com/Characters }
using { /Fortnite.com/Game }
using { /Fortnite.com/UI }
using { /Verse.org/Simulation }
using { /Verse.org/Simulation/Tags }
using { /Verse.org/Colors }
using { /UnrealEngine.com/Temporary/Diagnostics }
using { /UnrealEngine.com/Temporary/UI }
using { /UnrealEngine.com/Temporary/SpatialMath }

clash_powerup_manager_device := class(powerup_manager_device):

    OnBegin<override>()<suspends>:void=
        Sleep(9.0)
        (super:)OnBegin()
        GetPlayerDataDevice()

    var ClashPlayerDataDevice:clash_player_data_device=clash_player_data_device{}

    GetPlayerDataDevice():void=
        Devices:=FindCreativeObjectsWithTag(clash_player_data_device_tag{})
        for(Device:Devices):
            if:
                TargetPlayerDataDevice:=clash_player_data_device[Device]
            then:
                set ClashPlayerDataDevice = TargetPlayerDataDevice
                Print("Player Data Device Found!")
                return

            