using { /Fortnite.com/Devices }
using { /Verse.org/Simulation }
using { /Verse.org/Assets }
using { /UnrealEngine.com/Temporary/Diagnostics }
using { /Fortnite.com/Characters }

clash_powerup_max_damage_device := class(powerup_max_damage_device):

    BaseIncrementValue<override>:float=0.1
