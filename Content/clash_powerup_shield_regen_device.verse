using { /Fortnite.com/Devices }
using { /Verse.org/Simulation }
using { /Verse.org/Assets }
using { /UnrealEngine.com/Temporary/Diagnostics }
using { /Fortnite.com/Characters }

clash_powerup_shield_regen_device := class(powerup_shield_regen_device):

    BaseIncrementValue<override>:float=0.5
    RegenDelay<override>:float=2.0

