using { /Fortnite.com/Devices }
using { /Verse.org/Assets }
using { /Verse.org/Simulation }
using { /UnrealEngine.com/Temporary/Diagnostics }

# Add materials for each team here. 
FindMaterialsLineProps(Team:team_id)<computes>: material=
    case(Team):
        team_id.Blue => Materials.Identifiers.M_Blue_Line
        team_id.Red => Materials.Identifiers.M_Red_Line
        team_id.Neutral => Materials.Identifiers.M_Neutral_Line

FindMaterialsLogoProps(Team:team_id)<computes>: material=
    case(Team):
        team_id.Blue => Materials.Identifiers.M_Blue_Logo
        team_id.Red => Materials.Identifiers.M_Red_Logo
        team_id.Neutral => Materials.Identifiers.M_Neutral_Logo

<#
Set the color of the team that cleans the room.
There are three ways:
    - Props that change a team's material.
    - Team color lights.
#>
clash_room_color_manager := class<concrete>:
    @editable
    LinePropsByMaterial:[]creative_prop = array{}
    
    @editable
    LogoPropsByMaterial:[]creative_prop = array{}

    @editable
    RoomLightManager:room_light_manager = room_light_manager{}    

    ApplyTeamColorToRoom(Team:team_id): void=
        SetTeamMaterials(Team) 
        RoomLightManager.SetTeamLights(Team)               
        
    SetTeamMaterials(Team:team_id):void =
        for (LineProp : LinePropsByMaterial):
            SetLineMaterial(LineProp,Team)
        for (LogoProp : LogoPropsByMaterial):
            SetLogoMaterial(LogoProp,Team)

    SetLineMaterial(Prop:creative_prop, Team:team_id): void=  
        Prop.SetMaterial(FindMaterialsLineProps(Team))

    SetLogoMaterial(Prop:creative_prop, Team:team_id): void=  
        Prop.SetMaterial(FindMaterialsLogoProps(Team))

<#
Change the lights of each team.
#>
room_light_manager := class<concrete>:
    @editable
    BlueLights:[]customizable_light_device = array{}

    @editable
    RedLights:[]customizable_light_device = array{}

    @editable
    NeutralLights:[]customizable_light_device = array{}

    SetTeamLights(Team:team_id):void =
        TeamIndex := GetTeamIndex(Team)
        case(Team):
            team_id.Blue =>
                ToggleLights(BlueLights, true)
                ToggleLights(RedLights, false)
                ToggleLights(NeutralLights, false)
            team_id.Red =>
                ToggleLights(BlueLights, false)
                ToggleLights(RedLights, true)
                ToggleLights(NeutralLights, false)
            team_id.Neutral =>
                ToggleLights(BlueLights, false)
                ToggleLights(RedLights, false)
                ToggleLights(NeutralLights, true)

    ToggleLights(Lights:[]customizable_light_device, State:logic):void=
        for (CurentLight : Lights):
            if(State?):
                CurentLight.TurnOn()
            else:
                CurentLight.TurnOff() 