using { /Fortnite.com/Devices }
using { /Verse.org/Simulation }
using { /Verse.org/Simulation/Tags }
using { /UnrealEngine.com/Temporary/Diagnostics }
using { /Fortnite.com/Characters }
using { /Fortnite.com/Game }

# Room tags.
ClashRoom_01_Tag := class(tag){}
ClashRoom_02_Tag := class(tag){}
ClashRoom_03_Tag := class(tag){}
ClashRoom_04_Tag := class(tag){}
ClashRoom_05_Tag := class(tag){}
ClashRoom_06_Tag := class(tag){}
ClashRoom_07_Tag := class(tag){}
ClashRoom_08_Tag := class(tag){}
ClashRoom_09_Tag := class(tag){}
ClashRoom_10_Tag := class(tag){}
ClashRoom_11_Tag := class(tag){}
ClashRoom_12_Tag := class(tag){}
ClashRoom_13_Tag := class(tag){}
ClashRoom_14_Tag := class(tag){}
ClashRoom_15_Tag := class(tag){}

GetClashRoomTag(RoomIndex:int)<decides><transacts>:tag = 
	case(RoomIndex):
		1  => ClashRoom_01_Tag{}
		2  => ClashRoom_02_Tag{}
		3  => ClashRoom_03_Tag{}
		4  => ClashRoom_04_Tag{}
		5  => ClashRoom_05_Tag{}
		6  => ClashRoom_06_Tag{}
		7  => ClashRoom_07_Tag{}
		8  => ClashRoom_08_Tag{}
		9  => ClashRoom_09_Tag{}
		10 => ClashRoom_10_Tag{}
		11 => ClashRoom_11_Tag{}
		12 => ClashRoom_12_Tag{}
		13 => ClashRoom_13_Tag{}
		14 => ClashRoom_14_Tag{}
		15 => ClashRoom_15_Tag{}

log_clash_room_device := class(log_channel){} # TODO JJ: Study how rooms work

ClashGetPlayerNameToMessage<localizes>(Agent:agent):message="{Agent}"

ClashGetUsername<localizes>(Player : player) : message = "{Player}"

<# 
Volume used for any room/aisle in the map
Allows to keep track of what room is the current player at
Has data for a specific room
#>
clash_room_device := class(creative_device):

	Logger:log = log{Channel := log_clash_room_device}

	@editable
	var CurrentTeam<protected>:team_id = team_id.Neutral
	
	@editable
	var DisplayName<public>:string = ""

	@editable 
	Volume:volume_device = volume_device{}
	
	@editable
	FrontDoor:clash_door_device = clash_door_device{}

	@editable
	BackDoor:clash_door_device = clash_door_device{}

	@editable
	RoomColorManager:clash_room_color_manager = clash_room_color_manager{}  

	@editable
	MaybeFrontRoom:?clash_room_device = false

	@editable
	MaybeBackRoom:?clash_room_device = false

	@editable
	UnlockComponent<protected>:clash_room_enemy_device = clash_room_enemy_device{}

	#Reward System
	@editable
	RoomRewardDevice: room_reward_device = room_reward_device{}

	# TODO JJ need to check this
	@editable # Stage used for enemies in this room. INDEX_NONE if lobby. Use a different stage for each room
	Index<protected>:int = INDEX_NONE

	@editable
	ResetTimer:timer_device = timer_device{}

	@editable
	bDebug:logic = false

	# Its the room contested ?
	@editable
	var IsContested:logic = false

	@editable
	DEBUG_SetBlueButton:button_device = button_device{}

	@editable
	DEBUG_SetRedButton:button_device = button_device{}

	@editable
	DEBUG_FeedbackMessage:hud_message_device = hud_message_device{}

	#Simple common messages UI Stuff
	@editable
	UIEventMessageDisplayerDevice: clash_ui_event_displayer_device = clash_ui_event_displayer_device{}

	var HUDManager : hud_manager = hud_manager{}

	var PlayersInVolume:[]agent = array{}

	var PlayersPerTeam:[team_id]int = map:
		team_id.Red => 0
		team_id.Blue => 0

	var WasDisputed:logic = false

	#Clash event
	var RedTeamKills : int = 0

	var BlueTeamKills : int = 0

    var SubscribedSpawners : []guard_spawner_device = array{}

    var CancellableArray : []cancelable = array{}

    var CurrentRoom : string = ""

	SetHUDManager<public>(NewManager : hud_manager) : void = set HUDManager = NewManager

	OnAgentEnteredRoomEvent<public>:event(tuple(player, string)) = event(tuple(player, string)){} #event(agent) = event(agent){} 

	OnAgentExitedRoomEvent<public>:event(agent) = event(agent){}

	OnRoomTeamChanged<public>:event() = event(){}

#------------------------------------------------------------------------------------------------------------
	OnBegin<override>()<suspends>:void =
		if (DisplayName = ""):
			set DisplayName = "{Index}"

		if (bDebug?) then Logger.Print("Initialization of our room {DisplayName}")
		Volume.AgentEntersEvent.Subscribe(OnPlayerEntersRoom)
		Volume.AgentExitsEvent.Subscribe(OnPlayerExitsRoom)   
		ResetTimer.SuccessEvent.Subscribe(OnResetTimerFinished)

		UnlockComponent.Initialize(Index)

		#Initialize Reward Device
		RoomRewardDevice.Initialize(Index)

		AwaitKillsInRoom(UnlockComponent.Spawners, DisplayName)

		if (bDebug?):
			DEBUG_SetBlueButton.InteractedWithEvent.SubscribeAgent(DEBUG_SetTeam, team_id.Blue)
			DEBUG_SetRedButton.InteractedWithEvent.SubscribeAgent(DEBUG_SetTeam, team_id.Red)

		spawn{ AwaitEvents()}

#------------------------------------------------------------------------------------------------------------
	GetCurrentTeam()<decides><transacts> : team_id =
	{
		return CurrentTeam
	}

#------------------------------------------------------------------------------------------------------------
	OnPlayerEntersRoom(Agent:agent):void =
		if:
			Player := player[Agent]
			not PlayersInVolume.Find[Agent]
		then:
			set PlayersInVolume += array{Agent}
			# if (IsALobby[]):
			TrySpawnNextEnemies() # in the forward room !
			SpawnEnemies()
			TryAddPlayerInTeam(Player)

			BothTeamsInRoom := GetIfBothTeamsInRoom()
			if(BothTeamsInRoom?):
				for (PlayerInVolume : PlayersInVolume):
					UIEventMessageDisplayerDevice.DisplayEnemyInRoomMessage(PlayerInVolume)
					if (UnlockComponent.CurrentTeam = team_id.Neutral and not (DisplayName = "Red Lobby" or DisplayName = "Blue Lobby")):
						UIEventMessageDisplayerDevice.DisplayRoomClashMessage(PlayerInVolume)					
					if(EnteredPlayer := player[PlayerInVolume]):						
						spawn{HUDManager.DisplayClashEvent(EnteredPlayer, UnlockComponent.CurrentTeam, DisplayName)}
						HUDManager.UpdateClashEventInfo(EnteredPlayer, UnlockComponent.OriginalAgentAmount.Length, DisplayName, UnlockComponent.CurrentTeam, BlueTeamKills, RedTeamKills)
						

			if: # Check here old contested logic
				CurrentTeam = team_id.Neutral
				UnlockComponent.IsInTempMode[]
			then:
				ResetTimer.Reset()
				UnlockComponent.DisableTempMode()
				set WasDisputed = true

			var MatchPlayerTeam:logic=false

			if:
				PlayerTeamID := GetAgentTeam(Agent)
				PlayerTeamID = CurrentTeam
			then:
				set MatchPlayerTeam = true

			#If is room zero
			if (Index < 0):
				if:
					IsRoomZero:=RoomRewardDevice.GetIsRoomZero()
					MatchPlayerTeam?
					IsRoomZero = true
					RoomRewardDevice.HavePendingReward(Player) = false
				then:
					RoomRewardDevice.GrantReward(Player)
			else:
				if:
					MatchPlayerTeam?
					RoomRewardDevice.HavePendingReward(Player) = true
				then:
					#Grant Reward to the player
					RoomRewardDevice.GrantReward(Player)
				else:
					#If the player has no pending reward, hide the reward UI.
					RoomRewardDevice.RewardManager.HideAnyPreviewRoomReward(Player)
					
			if (bDebug?):
				# Username:string = Localize(GetUsername(Player))
				Logger.Print("Player enters room {DisplayName}")
		
			#DEBUG_ShowFeedback(Agent, UnlockComponent.GetTeamEliminations())

			if:
				PlayerEntering:= player[Agent]
			then:
				OnAgentEnteredRoomEvent.Signal(PlayerEntering, DisplayName)

#------------------------------------------------------------------------------------------------------------
	OnPlayerExitsRoom(Agent:agent):void =
		if:
			Player := player[Agent]
		then:
			# Username : string = Localize(GetUsername(Player))
			# if (bDebug?) then Logger.Print("Player {Username} leaves room {DisplayName}")

			set PlayersInVolume = PlayersInVolume.RemoveAllElements(Agent)

			BothTeamsInRoom := GetIfBothTeamsInRoom()
			if(PlayersInVolume.Length > 0 and not BothTeamsInRoom?):
				for (PlayerInVolume : PlayersInVolume):
					UIEventMessageDisplayerDevice.DisplayEnemyOutRoomMessage(PlayerInVolume)
					if (PlayerOnRoom := player[PlayerInVolume]):
						spawn{HUDManager.HideClashEvent(PlayerOnRoom)}
				UIEventMessageDisplayerDevice.DisplayEnemyOutRoomMessage(Agent)
				spawn{HUDManager.HideClashEvent(Player)}
			
			if: 
				not IsALobby[]
				CurrentTeam <> team_id.Neutral
				BeforeRoom := GetBackwardRoom(CurrentTeam)
				not BeforeRoom.IsALobby[]
				BeforeRoom.IsEmpty[]
			then:
				BeforeRoom.DespawnEnemies()

			TryRemovePlayerInTeam(Player)

			if:
				not IsALobby[]
				CurrentTeam = team_id.Neutral
			then:
				
				if (WasDisputed?, IsEmpty[]):
					ResetTimer.Start()

			# TODO : concede kills
			#BalanceKillsPerTeam(GetAgentTeam(Agent))
		

	<#OnPlayerEliminated(Result:elimination_result):void =
		if (bDebug?):
			Logger.Print("************* clash room device: Player eliminated **********") #thanos !
			# We want to reset rooms, but only for the team which player died 
			# First lets get the team of the dead person

			# now lets reset only the desired rooms of the proper team
			
			if:
				DestroyerTeam := GetAgentTeam(DestroyerAgent)            
				AgentsOnDestroyerTeam := GetAgentsInTeam[DestroyerTeam]
			then:
				RoomManager.RestartAllRooms()#>


#------------------------------------------------------------------------------------------------------------
	OnResetTimerFinished(MaybeAgent:?agent):void = # JJ TODO : Check if its beign used
		UnlockComponent.Reset()


#------------------------------------------------------------------------------------------------------------
	IsAgentInRoom(Agent:agent)<transacts><decides>:void =
		PlayersInVolume.Find[Agent]

#------------------------------------------------------------------------------------------------------------
# TODO: Check
	IsEmpty()<transacts><decides>:void =
		PlayersInVolume.Length <= 0

#------------------------------------------------------------------------------------------------------------
	RoomHasAgentOfTeam(Team:team_id)<transacts>:logic =
		if:
			PlayersInVolume.Length <= 0
		then:
			return false
		else : 
			for (Player : PlayersInVolume):
				if (GetAgentTeam(Player) = Team):
					Logger.Print("RoomHasAgentOfTeam:: Found player of team {Team} ::")
					return true
		return false

	IsALobby()<decides><transacts>:void =
		Index = INDEX_NONE

	GetTeam()<transacts>:team_id =
		CurrentTeam

#------------------------------------------------------------------------------------------------------------
	# Called to update the team for the room and do any needed improvements
	UpdateTeam(NewTeam: team_id):void =
		Logger.Print("********************* UPDATING TEAM ! *********************") # Thanos

		set RedTeamKills = 0
		set BlueTeamKills = 0

		set UnlockComponent.OriginalAgentAmount = array{}

		set CurrentTeam = NewTeam
		# lets update the teams in the rooms
		if (bDebug?) then Logger.Print("Team of room {DisplayName} changed to {NewTeam} ")
		RoomColorManager.ApplyTeamColorToRoom(NewTeam)
		UpdateEnemyComponent(NewTeam)
		UpdateDoors(NewTeam)
		set WasDisputed = false

		BothTeamsInRoom := GetIfBothTeamsInRoom()
		if(not BothTeamsInRoom?):
			for (PlayerInVolume : PlayersInVolume):
				UIEventMessageDisplayerDevice.DisplayEnemyOutRoomMessage(PlayerInVolume)
				
				if(PlayerOnRoom := player[PlayerInVolume]):
					spawn{HUDManager.HideClashEvent(PlayerOnRoom)}

		if:
			CurrentTeam <> team_id.Neutral
			AgentsInTeam := GetAgentsInTeam[NewTeam]
			RivalTeamID := GetEnemyTeam(CurrentTeam)
			RivalTeam := GetAgentsInTeam[RivalTeamID]
		then:
			for(Agent : AgentsInTeam):
				UIEventMessageDisplayerDevice.SendRoomCapturedMessage(Agent)
				OnRoomTeamChanged.Signal()
			# for(AgentInRivalTeam : RivalTeam):
			# 	UIEventMessageDisplayerDevice.SendRoomCapturedByEnemyTeamMessage(AgentInRivalTeam, CurrentTeam)


#------------------------------------------------------------------------------------------------------------
	# Called to update the team for the room and do any needed improvements, without respawning the guards
	UpdateTeamWithoutRespawn(NewTeam: team_id):void =
		Logger.Print("********************* UPDATING TEAM ! *********************")

		set RedTeamKills = 0
		set BlueTeamKills = 0

		set CurrentTeam = NewTeam
		# lets update the teams in the rooms
		if (bDebug?) then Logger.Print("Team of room {DisplayName} changed to {NewTeam} ")
		RoomColorManager.ApplyTeamColorToRoom(NewTeam)
		# UnlockComponent.SetTeam(NewTeam) #UpdateEnemyComponent(NewTeam) # Lets only update the team value without respawning anyone
		UpdateDoors(NewTeam)
		set WasDisputed = false
		
#------------------------------------------------------------------------------------------------------------
	AddTeamAllowed(Team:team_id):void =
		for:  
			Door: array{FrontDoor, BackDoor}
		do:
			Door.AddAllowedTeam(Team)
		  
#------------------------------------------------------------------------------------------------------------
	RemoveTeamAllowed(Team:team_id):void =
			Door := GetForwardDoor(Team)
			Door.RemoveAllowedTeam(Team)
		
#------------------------------------------------------------------------------------------------------------
	IsOnlyOneTeamInRoom()<decides><transacts>:team_id = 
		# TODO : concede kills
		#var TeamsWithPlayers:int = 0
		var MaybeTeamWithPlayers:?team_id = false
		for:
			Team -> Players : PlayersPerTeam
			Players > 0
		do:
			# TODO : concede kills
			#set TeamsWithPlayers += 1
			set MaybeTeamWithPlayers = option{Team}

		# TODO : concede kills
		#if:
		#	TeamsWithPlayers > 1
		#then:
		#	set MaybeTeamWithPlayers = false

		MaybeTeamWithPlayers?

#------------------------------------------------------------------------------------------------------------
	IsMoreThanOneTeamInRoom()<decides><transacts>:void =
		var TeamsWithPlayers:int = 0
		for:
			Team -> Players : PlayersPerTeam
			Players > 0
		do:
			set TeamsWithPlayers += 1
		TeamsWithPlayers > 1
		
#------------------------------------------------------------------------------------------------------------
	IsRoomCleared()<decides><transacts>:void = 
		UnlockComponent.AreEnemiesCleared[]

#------------------------------------------------------------------------------------------------------------
	CheckRoomState():void =
		if (IsRoomCleared[] ):
				
			WinnerTeam := UnlockComponent.GetTeamWithMoreEliminations()
			UpdateTeam(WinnerTeam)
			if (bDebug?) then Logger.Print("Room {DisplayName} cleared")

			<#TO DO JJ: Creo que seria mejor que se disparara un evento al liberar una room por ejemplo en el room manager
			para no tener que poner una referencia del room reward device en cada room device
			#>

			#Show Reward to winning team
			if (ScoringTeam := GetAgentsInTeam[CurrentTeam]):
				for (Agent : ScoringTeam):
					if (Player := player[Agent]):
						if(IsAgentInRoom[Agent]):
							# Print("Agent in Room")
							RoomRewardDevice.GrantReward(Agent)		
						else:
							# Print("Agent out of Room")
							TargetTeam:=GetAgentTeam(Player)
							RoomRewardDevice.GenerateRandomPendingReward(Player)

#------------------------------------------------------------------------------------------------------------
	AwaitEvents()<suspends>:void =
		race:
			loop:
				UnlockComponent.OnRoomObjectiveCompletedEvent.Await()
				CheckRoomState()
			loop:
				Data := UnlockComponent.OnDataUpdatedEvent.Await()
			#		for (Agent : PlayersInVolume):
			#		DEBUG_ShowFeedback(Agent, Data)
	
#------------------------------------------------------------------------------------------------------------
	Reset():void =
		Logger.Print("********************* RESET IN ROOM {Index} *********************")

		case (CurrentTeam):
			team_id.Blue =>
				if:
					AgentsInTeam := GetAgentsInTeam[team_id.Blue]
				then:
					for(Agent : AgentsInTeam):
						RoomRewardDevice.RemovePlayerReward(Agent) #Remove Rewards for looser team

						if (not IsALobby[]):
							UIEventMessageDisplayerDevice.SendRoomLostMessage(Agent)
			team_id.Red =>
				if:
					AgentsInTeam := GetAgentsInTeam[team_id.Red]
				then:
					for(Agent : AgentsInTeam):
						RoomRewardDevice.RemovePlayerReward(Agent) #Remove Rewards for looser team

						if (not IsALobby[]):
							UIEventMessageDisplayerDevice.SendRoomLostMessage(Agent)
			_=>

		if (not IsALobby[]):	
			UpdateTeam(team_id.Neutral)
			UnlockComponent.Reset() # making sure We reset the 

#------------------------------------------------------------------------------------------------------------
	ResetWithoutRespawn():void =
		Logger.Print("********************* RESET IN ROOM WITHOUT DESPAWNING {Index} *********************")

		case (CurrentTeam):
			team_id.Blue =>
				if:
					AgentsInTeam := GetAgentsInTeam[team_id.Blue]
				then:
					for(Agent : AgentsInTeam):
						RoomRewardDevice.RemovePlayerReward(Agent) #Remove Rewards for looser team

						if (not IsALobby[]):
							UIEventMessageDisplayerDevice.SendRoomLostMessage(Agent)
			team_id.Red =>
				if:
					AgentsInTeam := GetAgentsInTeam[team_id.Red]
				then:
					for(Agent : AgentsInTeam):
						RoomRewardDevice.RemovePlayerReward(Agent) #Remove Rewards for looser team

						if (not IsALobby[]):
							UIEventMessageDisplayerDevice.SendRoomLostMessage(Agent)
			_=> 
		
		if (not IsALobby[]):
			UpdateTeamWithoutRespawn(team_id.Neutral)
			UnlockComponent.ResetWithoutRespawn() # making sure We reset the room without despawning the enemies, but changing their team

#------------------------------------------------------------------------------------------------------------
	GetForwardDoor(Team:team_id)<transacts>:clash_door_device = 
		if (Team = team_id.Blue):
			FrontDoor
		else:
			BackDoor

#------------------------------------------------------------------------------------------------------------
	GetBackwardDoor(Team:team_id)<transacts>:clash_door_device = 
		if (Team = team_id.Blue):
			BackDoor
		else:
			FrontDoor

#------------------------------------------------------------------------------------------------------------
	SpawnEnemies():void =
	   UnlockComponent.Activate()
		
#------------------------------------------------------------------------------------------------------------
	TrySpawnNextEnemies():void =
		if:
			ForwardRoom := GetForwardRoom(CurrentTeam)
			ForwardRoom.IsEmpty[]
			# ForwardRoom.GetTeam() = team_id.Neutral
		then:
			ForwardRoom.SpawnEnemies()
			if (bDebug?):
				Logger.Print("********* Spawned enemies in forward room **********")
		else:
			if (bDebug?):
				Logger.Print("********* enemies not spawned in forward room **********")

#------------------------------------------------------------------------------------------------------------
	DespawnEnemies():void =
		UnlockComponent.Despawn()
	
#------------------------------------------------------------------------------------------------------------
	GetForwardRoom(Team:team_id)<transacts>:clash_room_device = 
		if (Team = team_id.Blue, NextRoom := MaybeFrontRoom?):
			NextRoom
		else if(Team = team_id.Red, PreviousRoom := MaybeBackRoom?):
			PreviousRoom
		else:
			if (bDebug?):
				Logger.Print("Incorrect configuration getting forward room in {DisplayName}, returning empty room")
			clash_room_device{}

#------------------------------------------------------------------------------------------------------------
	GetBackwardRoom(Team:team_id)<transacts>:clash_room_device = 
		if (Team = team_id.Red, NextRoom := MaybeFrontRoom?):
			NextRoom
		else if(Team = team_id.Blue, PreviousRoom := MaybeBackRoom?):
			PreviousRoom
		else:
			if (bDebug?):
				Logger.Print("Incorrect configuration getting backward room in {DisplayName}, returning empty room")
			clash_room_device{}

#------------------------------------------------------------------------------------------------------------
	UpdateDoors(Team:team_id):void =
		if (Team <> team_id.Neutral):
			AddTeamAllowed(Team)
			RemoveTeamAllowed(GetEnemyTeam(Team))
		else:
			RemoveTeamAllowed(team_id.Red)
			RemoveTeamAllowed(team_id.Blue)

#------------------------------------------------------------------------------------------------------------
	UpdateEnemyComponent(Team:team_id):void =
		UnlockComponent.SetTeam(Team)
		if (Team <> team_id.Neutral):
			UnlockComponent.Activate()
			TrySpawnNextEnemies()
		else:
			UnlockComponent.Despawn()

#------------------------------------------------------------------------------------------------------------
	GetIndex()<transacts>:int = 
		Index

#------------------------------------------------------------------------------------------------------------
	TryAddPlayerInTeam(Player:player):void =
		if:
			PlayerTeam := GetAgentTeam(Player)
			PlayerTeam <> team_id.Neutral
			set PlayersPerTeam[PlayerTeam] += 1
		else:
			if (bDebug?):
				Logger.Print("Error adding team in room {DisplayName}")
		
#------------------------------------------------------------------------------------------------------------
	TryRemovePlayerInTeam(Player:player):void =
		if:
			PlayerTeam := GetAgentTeam(Player)
			PlayersPerTeam[PlayerTeam] > 0
			set PlayersPerTeam[PlayerTeam] -= 1
		else:
			if (bDebug?):
				Logger.Print("Error removing team in room {DisplayName}")

#------------------------------------------------------------------------------------------------------------
	GetNumberOfPlayersOfTeam(Team:team_id)<decides><transacts>:int =
		PlayersPerTeam[Team]

#------------------------------------------------------------------------------------------------------------
	GetIfBothTeamsInRoom(): logic =
		#Check if there are more than one player in room
		var BlueTeamInRoom : logic = false
		var RedTeamInRoom : logic = false
		if:
			PlayersInVolume.Length > 1
		then: #Check if there is at least one player of each team in room
			for (PlayerInVolume : PlayersInVolume):
				PlayerTeamID := GetAgentTeam(PlayerInVolume)
				case (PlayerTeamID):
					team_id.Blue =>
						set BlueTeamInRoom = true
					team_id.Red =>
						set RedTeamInRoom = true
					_=>
		if(BlueTeamInRoom? and RedTeamInRoom?): #display alert message
			return true
		else return false
	
#------------------------------------------------------------------------------------------------------------
	DEBUG_SetTeam(Agent:agent, Team:team_id):void =
		UpdateTeam(Team)
		# Lets try respawn by force the enemies in the proper team
		UnlockComponent.Despawn()
		UnlockComponent.ActivateByForce()
	
#------------------------------------------------------------------------------------------------------------
	DEBUG_ShowFeedback(Agent:agent, Data:[team_id]int):void =
		if:
			RedElim := Data[team_id.Red]
			BlueElim := Data[team_id.Blue]
		then:  
			DEBUG_FeedbackMessage.Show(Agent, StringToMessage("Red eliminations: {RedElim} \n Blue eliminations: {BlueElim}"), ?DisplayTime := 0.0)


#-----------------------------------------------------------------------------------------------------------------------
	AwaitKillsInRoom(EnemySpawners : []guard_spawner_device, RoomName : string) : void =
		if (SubscribedSpawners.Length > 0):
			for (SpawnerIndex -> CurrentSpawner : SubscribedSpawners):
				if (CancellableArray[SpawnerIndex].Cancel()):
			set SubscribedSpawners = array{}

		set SubscribedSpawners = EnemySpawners

		for (CurrentSpawner : SubscribedSpawners):
			CancelSubscription := CurrentSpawner.EliminatedEvent.Subscribe(UpdateKills)
			set CancellableArray += array{CancelSubscription}

#-----------------------------------------------------------------------------------------------------------------------
	UpdateKills(InteractionResult : device_ai_interaction_result) : void =
		var KillerTeam : team_id = team_id.Neutral

		if(BaseSpawner := SubscribedSpawners[0]):
			MaybeKillerAgent := InteractionResult.Source
			if (KillerAgent := MaybeKillerAgent?):
				set KillerTeam = BaseSpawner.GetAgentTeam(KillerAgent)

				case (KillerTeam):
					team_id.Red =>
						set RedTeamKills += 1
					team_id.Blue =>
						set BlueTeamKills += 1
					team_id.Neutral =>
						Print("QUACK ERROR: NON TEAM KILL")

		for (PlayerInVolume : PlayersInVolume):
			HUDManager.UpdateRoomKills(RedTeamKills, BlueTeamKills, PlayerInVolume, UnlockComponent.OriginalAgentAmount.Length)

#------------------------------------------------------------------------------------------------------------
	<#CheckIfRoomContested() : logic =
		
		IsMoreThanOneTeamInCurrentRoom : logic = IsMoreThanOneTeamInRoom[]
		# Only neutral rooms can be contested, leave if its not the case
		if:
			CurrentTeam <> team_id.Neutral
		then:
			set IsContested = false
			return IsContested

		# Check if there are players from different teams in the same room, automatically means the room is contested!
		if:
			IsMoreThanOneTeamInCurrentRoom?
		then:
			set IsContested = true
			return IsContested


		# Pending, lets check adjacent rooms
		if:

		then:


		return IsContested#>

#-----------------------------------------------------------------------------------------------------------------------
	# TODO : concede kills
	TryAssignKillsToOtherTeam(TeamThatLeft : team_id) : void =
		UnlockComponent.GiveEliminationsOfTeamToOtherTeam(TeamThatLeft)

		if:
			RedKills := UnlockComponent.TempEliminationsPerTeam[team_id.Red]
			set RedTeamKills = RedKills

		if:
			BlueKills := UnlockComponent.TempEliminationsPerTeam[team_id.Blue]
			set RedTeamKills = BlueKills

		for (PlayerInVolume : PlayersInVolume):
			HUDManager.UpdateRoomKills(RedTeamKills, BlueTeamKills, PlayerInVolume, UnlockComponent.OriginalAgentAmount.Length)

#-----------------------------------------------------------------------------------------------------------------------
	# Idea here is simple: If a room is left by a player of team "A", lets check if there are no players of that team room in question and or the adjacent ones
	# and if there are players of team "B", We will award them the kills
	<#BalanceKillsPerTeam(TeamThatLeft : team_id) : void =

		# Are there any members of of our team around in the room or adjacent rooms?, no need to move those kills, leave
		var OnlyOpposingTeamInForwardRoom : logic = false
		var OnlyOpposingTeamInBackwardRoom : logic = false
		var OnlyOpposingTeamInCurrentRoom : logic = false

		ForwardRoom := GetForwardRoom(CurrentTeam)
		BackwardRoom := GetBackwardRoom(CurrentTeam)

		if:
			#not ForwardRoom.IsALobby[]
			TeamPresentInForwardRoom := ForwardRoom.IsOnlyOneTeamInRoom[]
		then:
			if ((TeamPresentInForwardRoom <> TeamThatLeft) or ForwardRoom.IsEmpty[]):
				Logger.Print("BalanceKills :: ForwardRoom team {TeamPresentInForwardRoom} or forward room is empty")
				set OnlyOpposingTeamInForwardRoom = true

		if:
			#not BackwardRoom.IsALobby[]
			TeamPresentInBackwardRoom := BackwardRoom.IsOnlyOneTeamInRoom[]
		then:
			if ((TeamPresentInBackwardRoom <> TeamThatLeft) or BackwardRoom.IsEmpty[]):
				Logger.Print("BalanceKills :: BackwardsRoom team {TeamPresentInBackwardRoom} or backwards room is empty")
				set OnlyOpposingTeamInBackwardRoom = true


		if:
			CurrentTeam = team_id.Neutral
			not IsALobby[]
			TeamPresentInRoom := IsOnlyOneTeamInRoom[]
		then:
			if ((TeamPresentInRoom <> TeamThatLeft) or IsEmpty[]):
				Logger.Print("BalanceKills :: current Room team {TeamPresentInRoom} or current room is empty")
				set OnlyOpposingTeamInCurrentRoom = true

		if:
			OnlyOpposingTeamInForwardRoom?
			OnlyOpposingTeamInBackwardRoom?
			OnlyOpposingTeamInCurrentRoom?
		then:
			UnlockComponent.GiveEliminationsOfTeamToOtherTeam(TeamThatLeft)

			case (TeamThatLeft):
				team_id.Red =>
					set BlueTeamKills += RedTeamKills
					set RedTeamKills = 0
				team_id.Blue =>
					set RedTeamKills += BlueTeamKills
					set BlueTeamKills = 0
				team_id.Neutral =>
					Print("QUACK ERROR: NON TEAM KILL")
			
			for (PlayerInVolume : PlayersInVolume):
				HUDManager.UpdateRoomKills(RedTeamKills, BlueTeamKills, PlayerInVolume, UnlockComponent.OriginalAgentAmount.Length)#>