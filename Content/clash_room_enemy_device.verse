
using { /Fortnite.com/Devices }
using { /Verse.org/Simulation }
using { /UnrealEngine.com/Temporary/Diagnostics }
using { /Verse.org/Simulation/Tags }

NoneTag := class(tag){}

log_crash_room_enemy_device := class(log_channel){}

clash_room_enemy_device := class(creative_device):
	Logger:log = log{Channel := log_crash_room_enemy_device}

	@editable
	var Spawners:[]guard_spawner_device = array{}

	@editable
	var SpawnedAgents:[]agent = array{}

	var OriginalAgentAmount:[]agent = array{}

	@editable
	bDebug:logic = false

	@editable
	DEBUG_MakeGuardsNeutralButton:button_device = button_device{}

	@editable
	DEBUG_MakeGuardsBlueButton:button_device = button_device{}

	@editable
	DEBUG_MakeGuardsRedButton:button_device = button_device{}

	#@editable
	#DEBUG_NeutralClassSelector: class_and_team_selector_device = class_and_team_selector_device{}

	#VFX Display

	@editable
	TeamNeutralVfxDevice:visual_effect_powerup_device = visual_effect_powerup_device{}
	@editable
	TeamBlueVfxDevice:visual_effect_powerup_device = visual_effect_powerup_device{}
	@editable
	TeamRedVfxDevice:visual_effect_powerup_device = visual_effect_powerup_device{}

	# If flag is true, enemies will respawn in its enterity next time a player enters. Intended for cases of rooms that have been left alone with 
	# no players neither in the room or adjacent rooms
	var EnemiesShouldRespawnNextTime : logic = false

	#---------------------------------------------------------------------------------------------------------------------

	var Index:int = -1 

	var RoomTag:tag = NoneTag{}

	@editable
	var CurrentTeam:team_id = team_id.Neutral


	var CurrentEnemies:int = 0

	var MaxEnemies:int = 0

	var EliminationsPerTeam:[team_id]int = map:
		team_id.Blue => 0
		team_id.Red => 0

	var TempEliminationsPerTeam:[team_id]int = map:
		team_id.Blue => 0
		team_id.Red => 0

	var bUseTempEliminations:logic = false

	OnRoomObjectiveCompletedEvent<public>:event() = event(){}
	OnDataUpdatedEvent<public>:event([team_id]int) = event([team_id]int){}
   
	Initialize(RoomIndex:int):void = 
		# if (Tag := GetClashRoomTag[RoomIndex]):
		#	 set RoomTag = Tag 
		#	 Objs := FindCreativeObjectsWithTag(RoomTag)
		#	 set Spawners = for:
		#		 Obj : Objs
		#		 Spawner := guard_spawner_device[Obj]
		#	 do: 
		#		 Spawner.SpawnedEvent.Subscribe(OnGuardSpawned)
		#		 Spawner.EliminatedEvent.Subscribe(OnGuardEliminated)
		#		 set MaxEnemies += Spawner.GetSpawnLimit()
		#		 Spawner
		set Index = RoomIndex

		for:
			Spawner : Spawners
		do: 
			Spawner.SpawnedEvent.Subscribe(OnGuardSpawned)
			Spawner.EliminatedEvent.Subscribe(OnGuardEliminated)
			set MaxEnemies += Spawner.GetSpawnLimit()
			if (bDebug?):
				Logger.Print("Added {Spawner.GetSpawnLimit()} to max enemies in room {Index}")
		if (bDebug?):
			Logger.Print("Room {Index} enemy component initialized, Max enemies are {MaxEnemies}")

	# -----------------------------------------------------------------------------------------		
	OnBegin<override>()<suspends>:void =
		if (bDebug?):
			DEBUG_MakeGuardsNeutralButton.InteractedWithEvent.SubscribeAgent(ChangeTeamLivingAgentsForButton, team_id.Neutral)
			DEBUG_MakeGuardsBlueButton.InteractedWithEvent.SubscribeAgent(ChangeTeamLivingAgentsForButton, team_id.Blue)
			DEBUG_MakeGuardsRedButton.InteractedWithEvent.SubscribeAgent(ChangeTeamLivingAgentsForButton, team_id.Red)

	
	# -----------------------------------------------------------------------------------------
	Activate():void =
		Logger.Print("CurrentEnemies {CurrentEnemies}  < MaxEnemies {MaxEnemies}")
		if (bDebug?):
			Logger.Print("CurrentEnemies {CurrentEnemies}  < MaxEnemies {MaxEnemies}")

		# If the room has no enemies its fine to respawn. Now, its also fine to respawn if the room and its adjacent rooms had no players (in other word,
		#there was no one around thus no reason to mantain the bnumber of enemies)

		if (CurrentEnemies <= 0 or EnemiesShouldRespawnNextTime?): 
			Spawn()

	# ONLY FOR DEBUG, respawns by force al enemies in the room, no matter how many there are
	ActivateByForce():void =
		Spawn()

	ChangeTeamLivingAgentsForButton(AgentInteractingWithButton: agent , Team:team_id):void =
		ChangeTeamLivingAgents(Team)

	ChangeTeamLivingAgents(Team:team_id):void =
		for (Agent : SpawnedAgents):
			
			#DEBUG_NeutralClassSelector.ChangeTeam(Agent)
			#DisplayTeamVfx(Agent, team_id.Neutral)
			AssignToTeamByIndex(Agent, Team)
			DisplayTeamVfx(Agent, Team)


	Spawn():void =
		#set EnemiesShouldRespawnNextTime = false
		for:
			Spawner : Spawners
			Count := 0..(Spawner.GetSpawnLimit() - 1)
		do:
			Spawner.Spawn()
			Logger.Print("********************* SPAWNING PEOPLE {CurrentTeam} *********************") # Thanos

	Despawn():void =
		for (Spawner : Spawners):
			Spawner.Despawn()

	AssignToTeamByIndex(Agent: agent, TeamID:team_id):void =
		
		TeamCollection := GetPlayspace().GetTeamCollection()
		Teams := TeamCollection.GetTeams()

		Logger.Print("******** ASSIGN GUARD TO TEAM {TeamID}, having {Teams.Length } teams , index {GetTeamIndex(TeamID)}*********************")

		if:
			Team := Teams[GetTeamIndex(TeamID)]
			TeamCollection.AddToTeam[Agent, Team]
			#Agent # cuack
		# then:
		#	 Logger.Print("New Team assigned {ToString(TeamID)}")
		# else:
		#	 Logger.Print("Could not assign team")

	OnGuardSpawned(Agent:agent):void =
		set CurrentEnemies += 1
		if (CurrentTeam <> team_id.Neutral):
			AssignToTeamByIndex(Agent, CurrentTeam)
		DisplayTeamVfx(Agent, CurrentTeam)
		set SpawnedAgents += array{Agent}
		set OriginalAgentAmount += array{Agent}
		if (bDebug?):
			Logger.Print("Guard Spawned in room {Index}")

		if:
			AgentsInTeam := GetValidAgentsInTeam[CurrentTeam]
		then:
			Logger.Print("******** GUARDS IN TEAM {CurrentTeam} ::  {AgentsInTeam.Length }*********************")

	OnGuardEliminated(Result:device_ai_interaction_result):void =
		set CurrentEnemies -= 1
		if (bDebug?):
			Logger.Print("Guard eliminated on room {Index}")
			
		if:
			Eliminator := Result.Source?
			Team := GetAgentTeam(Eliminator)
			set EliminationsPerTeam[Team] += 1
			set TempEliminationsPerTeam[Team] += 1
		then:
			CheckState()
			OnDataUpdatedEvent.Signal(GetTeamEliminations())
		else:
			if (bDebug?):
				Logger.Print("Failed to set elimination record on guard elimination")

		if:
			DeadAgent := Result.Target?
		then:
			
			<#
			var NewSpawnedAgents:[]agent = array{}
			for (Agent : SpawnedAgents):
				if (Agent <> DeadAgent):
					set NewSpawnedAgents += array{Agent}#>

			NewSpawnedAgents := SpawnedAgents.RemoveAllElements(DeadAgent)

			#SpawnedAgents.RemoveAllElements()
			set SpawnedAgents = NewSpawnedAgents

	CheckState():void =
		if (AreEnemiesCleared[]):
			OnRoomObjectiveCompletedEvent.Signal()

	AreEnemiesCleared()<decides><transacts>:void =
		CurrentEnemies = 0
		set OriginalAgentAmount = array{}

	GetTeamWithMoreEliminations():team_id = 
		
		var WinningTeam:team_id = team_id.Neutral
		var MostEliminations:int = -1
		EliminationsMap := if (bUseTempEliminations?) then TempEliminationsPerTeam else EliminationsPerTeam
		for:
			Team -> Eliminations : EliminationsMap
			Eliminations > MostEliminations
		do:
			set MostEliminations = Eliminations
			set WinningTeam = Team

		Logger.Print("** CHECKING TEAM WITH MORE ELIMINATIONS {WinningTeam} **")
		WinningTeam

	SetTeam(NewTeam:team_id):void =
		set CurrentTeam = NewTeam
		ResetEliminations()
		OnDataUpdatedEvent.Signal(GetTeamEliminations())
		
	# Lets reset the room, despawning enemies and setting the team to neutral
	Reset():void =
		Logger.Print("** Reset enemies")
		SetTeam(team_id.Neutral)
		Despawn()
		ResetEliminations()
		DisableTempMode()

	# Lets reset the room, setting the team to neutral and the enemis as well without respawning them
	ResetWithoutRespawn():void =
		Logger.Print("** Reset enemies without respawn !")
		SetTeam(team_id.Neutral)
		ChangeTeamLivingAgents(team_id.Neutral)
		ResetEliminations()
		DisableTempMode()

	ResetEliminations()<transacts>:void =
		set EliminationsPerTeam = map:
			team_id.Blue => 0
			team_id.Red => 0

		set TempEliminationsPerTeam = map:
			team_id.Blue => 0
			team_id.Red => 0
	
	GiveEliminationsOfTeamToOtherTeam(Team:team_id)<transacts>:void =
		Logger.Print("GiveEliminationsOfTeamToOtherTeam :: Attempting ...")
		if:
			TeamEliminations := EliminationsPerTeam[Team]
			EnemyTeamEliminations := EliminationsPerTeam[GetEnemyTeam(Team)]

			# TODO : concede kills
			# Passing real eliminations as well causes issues, like a dying player 
			# giving its kills to the other team from afar. And the room is awarded 
			# to the far away even if it wasnt there 
			#set EliminationsPerTeam[Team] = 0
			#set EliminationsPerTeam[GetEnemyTeam(Team)] = TeamEliminations + EnemyTeamEliminations

			set TempEliminationsPerTeam[Team] = 0
			set TempEliminationsPerTeam[GetEnemyTeam(Team)] = TeamEliminations + EnemyTeamEliminations
			
			#Logger.Print("GiveEliminationsOfTeamToOtherTeam :: Giving {TeamEliminations} eliminations to team {GetEnemyTeam(Team)}")
			#Logger.Print("GiveEliminationsOfTeamToOtherTeam :: Total Kills {TempEliminationsPerTeam[GetEnemyTeam(Team)]} for team {GetEnemyTeam(Team)}")

	RestoreTempTeamEliminations()<transacts>:void =
		for:
			Team -> Eliminations : EliminationsPerTeam
			set TempEliminationsPerTeam[Team] = Eliminations
		do:

	EnableTempMode(Team:team_id):void =
		set bUseTempEliminations = true
		GiveEliminationsOfTeamToOtherTeam(Team)

	DisableTempMode():void =
		set bUseTempEliminations = false
		RestoreTempTeamEliminations()

	IsInTempMode()<decides><transacts>:void =
		bUseTempEliminations?

	GetTeamEliminations()<transacts>:[team_id]int =
		EliminationsMap := if (bUseTempEliminations?) then TempEliminationsPerTeam else EliminationsPerTeam

	#Display Team VFX
	#---------------------------------------------------------------------------------------------------------------------

	DisplayTeamVfx(Agent : agent, Team : team_id): void =
		
		case (Team):
			team_id.Blue =>
				TeamBlueVfxDevice.Spawn()
				TeamBlueVfxDevice.Pickup(Agent)
			team_id.Red =>
				TeamRedVfxDevice.Spawn()
				TeamRedVfxDevice.Pickup(Agent)
			_ =>
				TeamNeutralVfxDevice.Spawn()
				TeamNeutralVfxDevice.Pickup(Agent)

	RemoveNpcColor(Agent : agent): void =
		block:	



	