
using { /Fortnite.com/Devices }
using { /Verse.org/Simulation }
using { /Verse.org/Simulation/Tags }
using { /Fortnite.com/Playspaces }
using { /UnrealEngine.com/Temporary/Diagnostics }
using { /Verse.org/Random }
using { /Fortnite.com/Characters }

log_clash_room_manager := class(log_channel){}

<# 
Data for a player related to the rooms
#>
# clash_player_room_data := class:
	
#	 # Visited rooms
#	 var RoomIndexes<private>:[]int = array{}
#	 # Last room the player entered too
#	 var LastRoomIndex:int = INDEX_NONE

#	 # A goal is an specific room, the last room that has a meaning in the run
#	 #var bHasStartedGoal:logic = false
#	 #var bHasCompletedGoal:logic = false

#	 # Returns the index of the room the player is at. INDEX_NONE if none
#	 GetCurrentRoomIndex()<transacts>:int =
#		 if (LastIndex := RoomIndexes[RoomIndexes.Length-1]):
#			 return LastIndex
#		 INDEX_NONE

#	 # Adds a new room index, returns true if success
#	 AddRoomIndex(Index:int)<transacts><decides>:void =
#		 if (not RoomIndexes.Find[Index]):
#			 UpdateLastRoomIndex()
#			 set RoomIndexes += array{Index}
#		 else:
#			 false?

#	 # Removes a room index, returns true if success
#	 RemoveRoomIndex(Index:int)<transacts><decides>:void =
#		 if (RoomIndexes.Find[Index]):
#			 UpdateLastRoomIndex()
#			 set RoomIndexes = RoomIndexes.RemoveAllElements(Index)
#		 else:
#			 false?
		
#	 # Only update last room index if it is not INDEX_NONE
#	 UpdateLastRoomIndex()<transacts>:void =
#		 if (Index := GetCurrentRoomIndex(), Index <> INDEX_NONE):
#			 set LastRoomIndex  =Index

# A Verse-authored creative device that can be placed in a level
clash_room_manager := class(creative_device):
	Logger:log = log{Channel := log_clash_room_manager}

	@editable # Reference of all the rooms in the map.
	Rooms:[]clash_room_device = array{}

	@editable
	PlayerEliminationDevice:elimination_manager_device = elimination_manager_device{}

	@editable
	UIEventMessageDisplayerDevice: clash_ui_event_displayer_device = clash_ui_event_displayer_device{}

	@editable
	bDebug:logic = false

	@editable
    SpawnerTag:tag=player_spawner_device_tag{}

	@editable
    AudioManagerDevice: clash_audio_manager = clash_audio_manager{}	

	var BlueAlivePlayers : int = 0
	var RedAlivePlayers : int = 0

	OnRoomEnteredEvent:event(tuple(player, string)) = event(tuple(player, string)){}
	
	OnRoomsTeamsUpdatedEvent: event([]team_id) = event([]team_id){} # team_id

	OnBegin<override>()<suspends>:void =
		PlayerEliminationDevice.EliminatedEvent.Subscribe(OnPlayerEliminated)		
		#PlayerEliminationDevice.EliminationEvent.Subscribe(OnPlayerElimination)	
		for (Room : Rooms):
			spawn: # so it exists a thread for each player
				RoomEventsInitializer(Room)

		GetAndBindAllSpawnerDevices()
	
	OnPlayerElimination(MaybeAgent:?agent):void =
		#Player Killed Enemy
		if (Agent := MaybeAgent?):
			
			if(Player:=player[Agent]):				
				#Player Killed Enemy TO DO CHECK IF NOT SUICIDE / RESPAWN
				AudioManagerDevice.PlayAudio_PlayerKilledRival(Agent)				
				AgentTeam:= GetAgentTeam(Agent)
				if (AgentTeam = team_id.Blue): #Blue KO
					AudioManagerDevice.PlayAudio_BlueKO_1(Agent)
				else: #Red KO
					AudioManagerDevice.PlayAudio_RedKO_1(Agent)

	OnPlayerEliminated(Agent:agent):void =

		if(Player:=player[Agent]):
			
			AgentTeam:= GetAgentTeam(Agent)					

			#Enemy Killed Player			
			AudioManagerDevice.PlayAudio_RivalKilledPlayer(Agent)
			if:
				AgentTeam = team_id.Blue
			then:				
				Logger.Print("OnPlayerSpawnedEvent :: Bye bluey")
				if (BlueAlivePlayers > 0): # TODO JJ : Check if at any time its being called unexpectedly
					set BlueAlivePlayers -= 1

			if:
				AgentTeam = team_id.Red
			then:
				Logger.Print("OnPlayerSpawnedEvent :: Bye redhead")
				if (RedAlivePlayers > 0): # TODO JJ : Check if at any time its being called unexpectedly
					set RedAlivePlayers -= 1

			# #Refill rewards for Spawning player
			GeneratePendingRoomRewardForPlayer(Player, AgentTeam)

		if (bDebug?):
			Logger.Print("Player eliminated")
			Logger.Print("************* clash room device: Player eliminated **********")
			# We want to reset rooms, but only for the team which player died 
			# First lets get the team of the dead person
					
		
		if:
			AgentTeam := GetAgentTeam(Agent)
			AgentTeam <> 0
		then: 			
			Swipe:= RestartAllRoomsInTeam(AgentTeam) <# #>
			if(Swipe = true): #Team Swipe!
				AudioManagerDevice.PlayAudio_TeamSwipe(Agent)
				if (AgentTeam = team_id.Red):
					UIEventMessageDisplayerDevice.DisplayRedTeamSwipe()
				else:
					UIEventMessageDisplayerDevice.DisplayBlueTeamSwipe()
				

	#Initialize Events
	RoomEventsInitializer(Room: clash_room_device)<suspends>:void =
		sync:
			loop:
				EventInfo := Room.OnAgentEnteredRoomEvent.Await()
				OnRoomEnteredEvent.Signal(EventInfo)
			loop:
				EventInfo := Room.OnRoomTeamChanged.Await()
				OnRoomsTeamChanged()
				#OnRoomEnteredEvent.Subscribe(Test) # TODO JJ : Check if we can somehow subscribe to one of our custom events

	SetHUDManagerReferenceForRooms(HUDManager : hud_manager) : void =
		for (Room : Rooms):
			Room.SetHUDManager(HUDManager)

	RestartTeamRoomsToNeutral(TeamId:team_id):void = # TODO check if should be deprecated
		for:
			Room : Rooms
			Room.GetTeam() = TeamId
			Room.GetIndex() <> INDEX_NONE
		do:
			Room.Reset()

	RestartAllRooms():void =
		for (Room : Rooms):
			Room.Reset()
		OnRoomsTeamChanged()

#=========================================================================================================================
    GetAndBindAllSpawnerDevices():void=
        var SpawnerDevices:[]player_spawner_device=array{}
        CreativeObjects:=FindCreativeObjectsWithTag(SpawnerTag)
        for(CreativeObject:CreativeObjects, MaybeSpawnerDevice:=player_spawner_device[CreativeObject]):
            set SpawnerDevices+=array{MaybeSpawnerDevice}
            MaybeSpawnerDevice.SpawnedEvent.Subscribe(OnPlayerSpawnedEvent)
        Logger.Print("Spawners Found: {SpawnerDevices.Length}")

#=========================================================================================================================================
	OnPlayerSpawnedEvent(SpawnedAgent:agent):void=
		if(Player:=player[SpawnedAgent]):
			if:
				GetAgentTeam(SpawnedAgent) = team_id.Blue
			then:
				Logger.Print("OnPlayerSpawnedEvent :: Hi bluey")
				set BlueAlivePlayers += 1

			if:
				GetAgentTeam(SpawnedAgent) = team_id.Red
			then:
				Logger.Print("OnPlayerSpawnedEvent :: Hi redhead")
				set RedAlivePlayers += 1
#=========================================================================================================================================
	RestartAllRoomsInTeam(Team:team_id):logic =		
		
		Logger.Print("RestartAllRoomsInTeam :: Alive blue players {BlueAlivePlayers} Alive Red players {RedAlivePlayers} ")

		if:
			Team = team_id.Blue
			BlueAlivePlayers > 0
		then:
			return false

		if:
			Team = team_id.Red
			RedAlivePlayers > 0
		then:
			return false

		# TODO : concede kills
		# Lets ensure all neutral rooms with kills on team for our dying team are awared to the other one
		#for (Room : Rooms, Room.GetTeam() = team_id.Neutral):
		#	Room.TryAssignKillsToOtherTeam(Team)
					

		for (Room : Rooms, Room.GetTeam() = Team):
			Room.ResetWithoutRespawn() # We want to restart all rooms without despawning the NPCs
		OnRoomsTeamChanged()
		return true

	OnRoomEntered(Player:player, RoomName:string):void =
		OnRoomEnteredEvent.Signal(Player, RoomName)

	OnRoomsTeamChanged():void =
		Logger.Print("CHECKING TEAMS IN ROOMS")
		var Teams:[]team_id = array{}

		for (Room : Rooms, not Room.IsALobby[]):
			set Teams += array{Room.GetTeam()}

		OnRoomsTeamsUpdatedEvent.Signal(Teams)

	GeneratePendingRoomRewardForPlayer(Player:player, Team:team_id):void=
		Print("Refilling Room Rewards")
		for (Room : Rooms, Room.GetTeam() = Team):
			if(not Room.RoomRewardDevice.GetIsRoomZero()?):
				Print("Team Controlled, Room: {Room.DisplayName}")
				Room.RoomRewardDevice.GenerateRandomPendingReward(Player)
			else:
				#Resetting Player Room Zero Reward Data
				Room.RoomRewardDevice.RemovePlayerReward(Player)

	# Commented by aecheverri, refactoring so this is pushed down for now
	# @editable
	# NPCManager:clash_enemies_manager = clash_enemies_manager{}

	# # Map with the player and its data
	# var PlayerRoomData:[player]clash_player_room_data = map{}

	# #STUFF ADDED BY MARTIN!
	# #ToCheckWithEngineering Martin doesn't know what his doing
	# @editable
	# ItemGranters:[]item_granter_device = array{}
	# #STUFF ADDED BY MARTIN!

	# OnClashPlayerEnteredRoom:event(tuple(player, int)) = event(tuple(player, int)){}


	# # Runs when the device is started in a running game
	# OnBegin<override>()<suspends>:void=		
	#	 for (Index := 0..(Rooms.Length-1)):		   
	#		 spawn{AwaitRoomEvents(Index)} # JJ : We cant subscribe to events, so we have each room waiting, like a coroutine, waiting for events in the rooms

	#	 # TODO JJ CHECK: Importante checar lo que resta del begin play del room manager, que creo ahi se espera a que se maten los enemigos en un cuarto !

	# # TODO JJ : We need eventually to register players to create their data and keep track of it
	# RegisterPlayer(Player:player):void =
	#	 if (not PlayerRoomData[Player], set PlayerRoomData[Player] = clash_player_room_data{}):
	#		 # Validate if the player started inside a room
	#		 for (Index := 0..(Rooms.Length-1), Room := Rooms[Index]):
	#			 if (Room.IsAgentInRoom[Player]):
	#				 OnAgentEnteredRoom(Player, Index)
	#	 else:
	#		 Logger.Print("Player already registered !")

	# # TODO JJ : We need eventually to register players to create their data and keep track of it
	# UnregisterPlayer(Player:player):void =
	#	 if (PlayerRoomData[Player]):
	#		 CleaningLastRoomInPlayer(Player)

	#		 var NewMap:[player]clash_player_room_data = map{}
	#		 for (KeyPlayer -> Data : PlayerRoomData, KeyPlayer <> Player):
	#			 set NewMap = ConcatenateMaps(NewMap, map{KeyPlayer => Data})
	#		 set PlayerRoomData = NewMap

	# # TODO JJ : We need eventually to register players to create their data and keep track of it
	# #GetPlayerRoomData(Player:player)<decides><transacts>: player_room_data =
	# #	PlayerRoomData[Player]

	# # JJ: coroutine waiting for players (not NPCs) entering the room for example
	# AwaitRoomEvents(IndexRoom:int)<suspends>:void =
	#	 if (Room := Rooms[IndexRoom]):		   
	#		 sync:
	#			 loop:
	#				 Agent:agent = Room.OnAgentEnteredRoomEvent.Await()
	#				 OnAgentEnteredRoom(Agent, IndexRoom)

	#			 loop:
	#				 Agent:agent = Room.OnAgentExitedRoomEvent.Await()
	#				 OnAgentExitedRoom(Agent, IndexRoom)

	#			 loop:
	#				 Result := NPCManager.OnStageEmptiedByTeam.Await()
	#				 if:
	#					 Result(1) = Room.Index
	#				 then:
	#					 AssignRoomToTeam(Result(0), Room)

	# # A player (non NPC) has entered the room
	# OnAgentEnteredRoom(Agent:agent, RoomIndex:int):void =
		
	#	 if (Player := player[Agent]):
	#		 RegisterPlayer(Player)

	#	 if (Player := player[Agent], Data := PlayerRoomData[Player], PreviousRoomIndex := Data.GetCurrentRoomIndex(), Data.AddRoomIndex[RoomIndex]):
			
	#		 Logger.Print("On Agent Entered Room Index {RoomIndex} !!")
	#		 OnRoomIndexesChanged(Player, PreviousRoomIndex, Data.GetCurrentRoomIndex())

	#		 # Lets throw  our signal for trackers, other managers, etc, currently not being used
	#		 OnClashPlayerEnteredRoom.Signal((Player, RoomIndex))

	#		 # TODO JJ : Possibly We will remove the goal related portions, but will see about the rewards and items
	#		 # If room is any room before goal then we need to remove all rewards and clean player from room rewards
	#		 #if (RoomIndex < RoomIndexStartGoal):
	#		 #	AllItemsRemover.Remove(Player)
	#		 #	ResetPlayerReceivedRewards(Player) # Do not reset the rewards, it will only be done when dead
	#		 #	HideAnyPreviewRoomReward(Player)
	#		 #	OnPlayerEnteredRoomBeforeGoal.Signal(Player)

	#		 #
	#		 #if (bTrackPlayerGoal?):
	#		 #	if (Data.bHasStartedGoal?):
	#		 #		if(RoomIndex < RoomIndexStartGoal):
	#		 #			if (set PlayerRoomData[Player].bHasStartedGoal = false):
	#		 #				OnGoalCancelled.Signal(Player)

	#		 #				if (Data.bHasCompletedGoal?):
	#		 #					OnReplayGoal.Signal(Player)
	#		 #					RestartGoalInPlayer(Player)

	#		 #		if (not Data.bHasCompletedGoal? and RoomIndex >= RoomIndexCompleteGoal):
	#		 #			if (set PlayerRoomData[Player].bHasCompletedGoal = true):
	#		 #				OnGoalCompleted.Signal(Player)
	#		 #	else:
	#		 #		if (not Data.bHasCompletedGoal? and RoomIndex >= RoomIndexStartGoal and RoomIndex < RoomIndexCompleteGoal):
	#		 #			if (set PlayerRoomData[Player].bHasStartedGoal = true):
	#		 #				OnGoalStarted.Signal(Player)

	# # A player (non NPC) has left the room
	# OnAgentExitedRoom(Agent:agent, RoomIndex:int):void =

	#	 #Grant item when player exits room volume (-MARTIN)
	#	 GrantRandomItem(Agent)
		
	#	 # TODO JJ : if there is no player data, should We create it?, REGISTER THE PLAYER?
	#	 if (RoomIndex > INDEX_NONE, Player := player[Agent], Data := PlayerRoomData[Player], PreviousRoomIndex := Data.GetCurrentRoomIndex(), Data.RemoveRoomIndex[RoomIndex]):
			
	#		 Logger.Print("On Agent Exited Room Index {RoomIndex} !!")
	#		 OnRoomIndexesChanged(Player, PreviousRoomIndex, Data.GetCurrentRoomIndex())
	

	# # Here is where the magic happens and We will spawn our NPCs !		
	# # JJ: called whenever a player moves bettwen rooms, since changed the room, we verify again things to be done
	# OnRoomIndexesChanged(Player:player, PreviousRoomIndex:int, NewRoomIndex:int):void =
	#	 # Verify if current room changed, otherwise skip
	#	 if (PreviousRoomIndex = NewRoomIndex):
	#		 return

	#	 #UpdateRoomHUDForPlayer(Player)

	#	 if (NewRoomIndex > INDEX_NONE):
	#		 # Notify accolade TODO JJ check this
	#		 #if (NewRoom := Rooms[NewRoomIndex]):
	#		 #	NewRoom.PlayerEnteredRoomAnalytics.Submit(Player)

	#		 # TODO JJ CHECK HERE IS WHERE THE MAGIC WILL HAPPEN !
	#		 # Reset next room if player moving around rooms, or not locked # TODO JJ: making sure we activate the room anyway

	#		 # Get player team to decide in which direction go with rooms
	#		 TeamCollection := GetPlayspace().GetTeamCollection()
	#		 if (AgentsTeam := TeamCollection.GetTeam[Player]):
	#			 TeamArray := TeamCollection.GetTeams()
	#			 for (TeamNumber->Team : TeamArray):
	#				 if (AgentsTeam = Team):
	#					 Logger.Print("<<<<<<<<<<<< Team ? {TeamNumber}")
	#					 case(TeamNumber):
	#						 0 => # Team one, player one
	#							 if (NextRoom := Rooms[NewRoomIndex+1], ShouldResetRoom(Player, NextRoom, NewRoomIndex+1)?):	 
	#								 NPCManager.ActivateStage(NextRoom.Index)
	#						 1 => # Team two, player two
	#							 if (NextRoom := Rooms[NewRoomIndex-1], ShouldResetRoom(Player, NextRoom, NewRoomIndex-1)?):	 
	#								 NPCManager.ActivateStage(NextRoom.Index)
	#						 _ =>
	#							 Logger.Print("Something seems wrong, Team ? {TeamNumber}")
						

	#		 #if (NextRoom := Rooms[NewRoomIndex-1], ShouldResetRoom(Player, NextRoom, NewRoomIndex-1)?):	 
	#		 #	NPCManager.ActivateStage(NextRoom.NPCStageIndex) # JJ: when advancing trough rooms, withoiut having to kill enemies

	#		 # Try clean when enters or exists to ensure empty states are updated
	#		 # HERE CleaningLastRoomInPlayer(Player)

	#	 # TODO JJ: No work in rewards for now
	#	 #if (PreviousRoomIndex > INDEX_NONE):
	#	 #	HideAnyPreviewRoomReward(Player)

	#	 # TODO JJ: Here We will work with the locks and the rewards
	#	 # If room after start goal and next room has reward and no locker, then show preview
	#	 # If next room has lock and is already unlocked, then show preview
	#	 #if:
	#	 #	NewRoomIndex >= RoomIndexStartGoal
	#	 #	NextRoomIndex:= NewRoomIndex+1
	#	 #	NextRoom := Rooms[NextRoomIndex]
	#	 #	NextRoom.RoomRewards.Length > 0
	#	 #	not NextRoom.bUseLock? or not NextRoom.bIsLocked?
	#	 #then:
	#	 #	TryShowPreviewRoomReward(Player, NextRoomIndex, NextRoom)

	#		 # Cleans NPCs in player last room index JJ# called when player dies

	# CleaningLastRoomInPlayer(Player:player):void =
	#	 # Verify player last room and deactivate this and the next one (releases NPCs to have more space in the available pool (Max 30 in island))
	#	 if (Data := PlayerRoomData[Player], Data.LastRoomIndex <> INDEX_NONE, LastPlayerRoom := Rooms[Data.LastRoomIndex]):
	#		 Logger.Print("Try cleaning last room {Data.LastRoomIndex} and next last room {Data.LastRoomIndex+1}")

	#		 # TODO JJ CHECK: pending to do the cleaning here
	#		 #if (ShouldDeactivateRoom(LastPlayerRoom, Data.LastRoomIndex)?):
	#		 #	Logger.Print("Cleaning last room {Data.LastRoomIndex}")
	#		 #	NPCManager.DeactivateStage(LastPlayerRoom.NPCStageIndex)
			
	#		 #if (NextLastRoom := Rooms[Data.LastRoomIndex+1], ShouldDeactivateRoom(NextLastRoom, Data.LastRoomIndex+1)?):
	#		 #	Logger.Print("Cleaning next last room {Data.LastRoomIndex+1}")
	#		 #	NPCManager.DeactivateStage(NextLastRoom.NPCStageIndex)
		 
	# # TODO JJ: Here We need to look in deep, really deep detail what to do
	# ShouldResetRoom(Player:player, Room:clash_room_device, Index:int)<transacts>:logic =
		
	#	 # This is temporal!
	#	 return true

	#	 #if (HasAnyPlayerCurrentRoomIndex(Index)?): # JJ: if there is a plyer, lets not  reset the room, why if the player is there
	#	 #	return false

	#	 # Player is in INDEX_NONE and LastRoom is the previous one 
	#	 #for (PlayerData:PlayerRoomData):
	#	 #	if (PlayerData.GetCurrentRoomIndex() = INDEX_NONE): # JJ: the player is in no room, is in the empty space bettwen rooms, not reset the room, lets let they to shot form outside
	#	 #		if (PlayerData.LastRoomIndex = Index or PlayerData.LastRoomIndex = (Index-1)):
	#	 #			return false

	#	 #if (Room.bUseLock?): # JJ: its the kidn of room that has a lock
	#	 #	if (Room.bIsLocked?):
	#	 #		return false

	#	 #	# Validate no players in previous room if I was previously in this room
	#	 #	if (PlayerRoomData[Player].LastRoomIndex = Index):
	#	 #		return false
	#	 #	# If has any player besides me then do not reset
	#	 #	if (HasAnyPlayerCurrentRoomIndexBesidesPlayer(Index-1, Player)?): # im in room 11, i return to a stairs (-1), i dotn want to reset that last room
	#	 #		return false
	#	 # If not initial room # JJ: special rooms without locks, like lobby, stairs, etc
	#	 #else if ((Index-1) >= RoomIndexStartGoal):
	#	 #	if (PlayerRoomData[Player].LastRoomIndex = Index and HasAnyPlayerCurrentRoomIndex(Index-1)?):
	#	 #		return false
	#	 #	if (HasAnyPlayerCurrentRoomIndexBesidesPlayer(Index-1, Player)?):
	#	 #		return false

	#	 #return true
		
	# Change the color of the room to the team that cleans the room.
	# AssignRoomToTeam(TeamId:team_id,ClashRoom:clash_room_device): void=
	#	 ClashRoom.UpdateTeam(TeamId)   


	# #CODE by MARTIN				
	# GrantRandomItem(Agent:agent) : void=
	# 	var ItemGranterIndex: int = GetRandomInt(0, ItemGranters.Length)
	# 	if(Selected_ItemGranter := ItemGranters[ItemGranterIndex]):
	# 		Selected_ItemGranter.CycleToRandomItem(Agent)
	# 		Logger.Print("Granting Random Item to Player")
	# #CODE by MARTIN
