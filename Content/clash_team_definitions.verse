using { /Fortnite.com/Characters }
using { /Fortnite.com/Devices }
using { /Verse.org/Simulation }

<# 
TEAMS
DO NOT change the order after island is published
Order comes from the defined in Team Settings device
UPDATE ToTeamId() and GetTeamIndex() when you add more
#>
team_id := enum:
    Blue #0
    Red #1
    Neutral #2
    # ...

# ADD to this array the available teams
AllTeamIds:[]team_id = array:
    team_id.Blue,
    team_id.Red,
    team_id.Neutral
    # ...

# DO NOT change the order after island is published
ToTeamId(Index:int)<computes>: team_id =
    case(Index):
        0 => team_id.Blue
        1 => team_id.Red
        2 => team_id.Neutral
        # ...

        _ => team_id.Neutral

# DO NOT change the order after island is published
GetTeamIndex(Id:team_id)<computes>: int =
    case(Id):
        team_id.Blue => 0 # 0
        team_id.Red => 1 # 1
        team_id.Neutral => 2 # 2
        # ...

ToString(Id:team_id)<computes>:string =
    case(Id):
        team_id.Blue => "Blue"
        team_id.Red => "Red"
        team_id.Neutral => "Neutral"
        # ...

GetEnemyTeam(Team:team_id)<computes>: team_id =
    case(Team):
        team_id.Blue => team_id.Red
        team_id.Red => team_id.Blue
        team_id.Neutral => team_id.Neutral

# Returns the team Id for a specific agent
(CreativeObject:creative_object_interface).GetAgentTeam(Agent:agent)<transacts>: team_id =
    TeamCollection := CreativeObject.GetPlayspace().GetTeamCollection()
    for (Index -> Team : TeamCollection.GetTeams(), TeamCollection.IsOnTeam[Agent, Team]):
        return ToTeamId(Index)

    Print("ERROR - Team not found for agent")
    return team_id.Neutral

(CreativeObject:creative_object_interface).GetAgentsInTeam(Team:team_id)<transacts><decides>: []agent =
    TeamCollection := CreativeObject.GetPlayspace().GetTeamCollection()
    TeamIndex := GetTeamIndex(Team)
    TeamCollection.GetAgents[TeamCollection.GetTeams()[TeamIndex]]

(CreativeObject:creative_object_interface).GetAliveAgentsInTeam(Team:team_id)<transacts><decides>: []agent =
	TeamCollection := CreativeObject.GetPlayspace().GetTeamCollection()
    TeamIndex := GetTeamIndex(Team)
    Agents := TeamCollection.GetAgents[TeamCollection.GetTeams()[TeamIndex]]

	var AliveAgents :[]agent = array{}

	for (Agent : Agents, Player := player[Agent], FortCharacter := Player.GetFortCharacter[]):
		if(PlayerHealth := FortCharacter.GetHealth() > 0.0):
			Print("Life for character: {PlayerHealth} ")
			set AliveAgents += array{Agent}
	Print("How many alive: {AliveAgents.Length}")
	Agents
	
# Success if there are more players than agent in a specific team, if not a player then fails
(CreativeObject:creative_object_interface).AreThereMorePlayersInTeam(Agent:agent)<transacts>: logic =
    if (player[Agent], Agents:=CreativeObject.GetAgentsInTeam[CreativeObject.GetAgentTeam(Agent)]):
        for (OtherAgent:Agents, OtherAgent <> Agent and player[OtherAgent]):
            return true
    return false

# Returns the agents that are valid, ignores inactive (dead) NPCs
(CreativeObject:creative_object_interface).GetValidAgentsInTeam(Team:team_id)<transacts><decides>: []agent =
    AgentsToCheck:=CreativeObject.GetAgentsInTeam[Team]
    # If Player or if NPC Active
    for (Agent:AgentsToCheck, player[Agent] or Agent.GetFortCharacter[].IsActive[]):
        Agent

# Success if all members in a team are dead
(CreativeObject:creative_object_interface).AreAllTeamAgentsDead(Team:team_id)<transacts>: logic =
    if (Agents:=CreativeObject.GetAgentsInTeam[Team]):
        for (Agent:Agents):
            if (Fort:=Agent.GetFortCharacter[], Fort.IsActive[]):
                return false
    return true

team_definition := struct<concrete>:
    @editable
    Id:team_id = team_id.Blue

    @editable
    DisplayName:string = ""

    # ... Add here if you need more configurations for the team

<# 
General definitions for teams
#>
clash_team_definitions := class(creative_device):
    @editable
    TeamDefinitions:[]team_definition = array{}

    FindTeamDefinition(Id:team_id)<decides><transacts>:team_definition =
        var MaybeTeamDefinition:?team_definition = false
        for (TeamDefinition:TeamDefinitions, not MaybeTeamDefinition?):
            if (TeamDefinition.Id = Id):
                set MaybeTeamDefinition = option{TeamDefinition}
        MaybeTeamDefinition?