
using { /Fortnite.com/Devices }
using { /Verse.org/Simulation }
using { /UnrealEngine.com/Temporary/Diagnostics }

clash_team_vfx_assignation_device := class(creative_device):

	@editable
	VFXDevice: visual_effect_powerup_device = visual_effect_powerup_device{}

	@editable
	VFXBlueHat: visual_effect_powerup_device = visual_effect_powerup_device{}

	@editable
	VFXRedHat: visual_effect_powerup_device = visual_effect_powerup_device{}

	@editable
	SpawnersCollection: []player_spawner_device = array{}

	@editable
	TeamSelectorCollection: []class_and_team_selector_device = array{}
	@editable
	IsHatVFXEnabled: logic = false
	#-----------------------------------------------------------------------------------------------------------------------------------------

	OnBegin<override>()<suspends>:void =
		# for (Player : GetPlayspace().GetPlayers()):
		#	 AssignVFX(Player)

		for (Spawner: SpawnersCollection):
			Spawner.SpawnedEvent.Subscribe(AssignVFX)

		for (TeamSelector : TeamSelectorCollection):
			TeamSelector.TeamSwitchedEvent.Subscribe(AssignVFX)

		# GetPlayspace().PlayerAddedEvent().Subscribe(AssignVFX)

	AssignVFX(Agent: agent): void =
		VFXDevice.Spawn()
		VFXDevice.Pickup(Agent)
		if (IsHatVFXEnabled = true):			
			if:
				GetAgentTeam(Agent) = team_id.Blue
			then:				
				VFXBlueHat.Spawn()
				VFXBlueHat.Pickup(Agent)
			if:
				GetAgentTeam(Agent) = team_id.Red
			then:
				VFXRedHat.Spawn()
				VFXRedHat.Pickup(Agent)
				

		