using { /Fortnite.com/Devices }
using { /Verse.org/Simulation }
using { /UnrealEngine.com/Temporary/Diagnostics }
using { /UnrealEngine.com/Temporary/UI }
using { /Fortnite.com/UI }
using { /Verse.org/Colors/NamedColors }
using { /UnrealEngine.com/Temporary/SpatialMath }
using { /Verse.org/Colors }

#Classes
#------------------------------------------------------------------------------------------------------------
clash_ui := class:

    var UICanvas : canvas_slot = canvas_slot{Widget:=stack_box{Orientation:=orientation.Vertical}}
    var bIsVisible : logic = false

    CreateUI():canvas_slot =
        return UICanvas

#------------------------------------------------------------------------------------------------------------
    SetVisibility(bVisible : logic): void =
        if(bVisible?):
            UICanvas.Widget.SetVisibility(widget_visibility.Visible)
            set bIsVisible = true
        else:
            UICanvas.Widget.SetVisibility(widget_visibility.Hidden)
            set bIsVisible = false

#------------------------------------------------------------------------------------------------------------
    DisplayForSeconds(DisplayTime : float)<suspends>: void =
        
        UICanvas.Widget.SetVisibility(widget_visibility.Visible)
        Sleep(DisplayTime)
        UICanvas.Widget.SetVisibility(widget_visibility.Hidden)

#------------------------------------------------------------------------------------------------------------
    GetVisibility(): logic =
        return bIsVisible