
using { /Fortnite.com/Devices }
using { /Verse.org/Simulation }
using { /Fortnite.com/Characters }
using { /UnrealEngine.com/Temporary/Diagnostics }
using { /Fortnite.com/Game }
using { /Verse.org/Simulation/Tags }
clash_ui_event_displayer_device := class(creative_device):

    @editable
    HUDCommonMessageDevice:hud_message_device = hud_message_device{}
    @editable
    HUDBlueCommonMessageDevice:hud_message_device = hud_message_device{}
    @editable
    HUDRedCommonMessageDevice:hud_message_device = hud_message_device{}
    # -- BED UNDER ATTACK --
    @editable
    HUDBlueBedUnderAttackMessageDevice:hud_message_device = hud_message_device{}
    @editable
    HUDRedBedUnderAttackMessageDevice:hud_message_device = hud_message_device{}
    # -- ROOM CONTROL --
    @editable
    HUDBlueRoomMessageDevice:hud_message_device = hud_message_device{}
    @editable
    HUDRedRoomMessageDevice:hud_message_device = hud_message_device{}
    @editable
    HUDRoomLostMessageDevice:hud_message_device = hud_message_device{}   
    # -- RESPAWN & Objective--
    SpawnerTag: tag = player_spawner_device_tag{}
    @editable
    HUDRespawnMessageMessageDevice:hud_message_device = hud_message_device{}
    @editable
    HUDDestroyBedsMessageMessageDevice:hud_message_device = hud_message_device{}
    @editable
    HUDMatchStartMessageDevice:hud_message_device = hud_message_device{}
    @editable
    HUDChooseTeamMessageDevice:hud_message_device = hud_message_device{}
    @editable
    HUDRedTeamMessageDevice:hud_message_device = hud_message_device{}
    @editable
    HUDBlueTeamMessageDevice:hud_message_device = hud_message_device{}
    # -- KO & Swipe--
    @editable
    HUD_RED_KO_MessageDevice:hud_message_device = hud_message_device{}
    @editable
    HUD_BLUE_KO_MessageDevice:hud_message_device = hud_message_device{}
    @editable
    HUD_RED_SWIPE_MessageDevice:hud_message_device = hud_message_device{}
    @editable
    HUD_BLUE_SWIPE_MessageDevice:hud_message_device = hud_message_device{}
    #-----Audio-----   
    @editable
    AudioManagerDevice: clash_audio_manager = clash_audio_manager{}

    @editable #On player death
    HUDWastedCustomMessageDevice:hud_message_device = hud_message_device{} #Implemented
    @editable #On player in the same room
    HUDEnemyInRoomCustomMessageDevice:hud_message_device = hud_message_device{} #Implemented
    #@editable #On player in the same room
    #HUDEnemyOutRoomCustomMessageDevice:hud_message_device = hud_message_device{} #Implemented
    
    var IsWastedEnabled: logic  = false
    #-------------------------------------------------------------------------------------------------
    OnBegin<override>()<suspends>: void =
        AllPlayers: []player = GetPlayspace().GetPlayers()
        for (Player : AllPlayers):
            SubscribePlayerToEvents(Player)

        GetPlayspace().PlayerAddedEvent().Subscribe(SubscribePlayerToEvents)        
        SubscribeSpawners()
    #Common Messages  
   
    
    #-------------------------------------------------------------------------------------------------
    SendEnemyBedDamagedMessage(Agent: agent, RivalTeam : team_id): void =
        case(RivalTeam):
            team_id.Blue =>                
                AudioManagerDevice.PlayAudio_BlueBedUnderAttack(Agent)
                HUDBlueBedUnderAttackMessageDevice.Show(Agent)
            team_id.Red =>            
                AudioManagerDevice.PlayAudio_RedBedUnderAttack(Agent)
                HUDRedBedUnderAttackMessageDevice.Show(Agent)
            _=>
        #HUDBlueBedUnderAttackMessageDevice.SetText(StringToMessage("{RivalTeam}" + EnemyBedDamagedText))
        

    #-------------------------------------------------------------------------------------------------
    SendRoomCapturedMessage(Agent: agent): void =
        PlayerTeam := GetAgentTeam(Agent)
        case(PlayerTeam):
            team_id.Blue =>                
                HUDBlueRoomMessageDevice.Show(Agent)
                AudioManagerDevice.PlayAudio_Blue_Room(Agent)
            team_id.Red =>                
                HUDRedRoomMessageDevice.Show(Agent)
                AudioManagerDevice.PlayAudio_Red_Room(Agent)
            _=>
        AudioManagerDevice.PlayAudio_CapturedRoom(Agent)

    #-------------------------------------------------------------------------------------------------
    SendRoomLostMessage(Agent: agent): void =
        PlayerTeam := GetAgentTeam(Agent)        
        HUDRoomLostMessageDevice.Show(Agent)

    #-------------------------------------------------------------------------------------------------
    # SendNewRoundMessage(Agent: agent): void =
    #     HUDCommonMessageDevice.SetText(StringToMessage(NewRoundText))
    #     HUDCommonMessageDevice.Show(Agent)

    #Custom Designed Messages
    #-------------------------------------------------------------------------------------------------
    DisplayRedTeamHUD(Agent: agent): void =
        HUDRedTeamMessageDevice.Show(Agent) 
    #-------------------------------------------------------------------------------------------------
    DisplayBlueTeamHUD(Agent: agent): void =
        HUDBlueTeamMessageDevice.Show(Agent)         
    #-------------------------------------------------------------------------------------------------
    DisplayWastedMessage(Agent: agent): void =
        HUDWastedCustomMessageDevice.Show(Agent)        
    #-------------------------------------------------------------------------------------------------
    DisplayMatchStartMessage(Agent: agent): void =
        HUDMatchStartMessageDevice.Show(Agent)        
    #-------------------------------------------------------------------------------------------------
    DisplayChooseTeamMessage(Agent: agent): void =
        HUDChooseTeamMessageDevice.Show(Agent)
        AudioManagerDevice.PlayAudio_SelectYourTeam(Agent)
    #-------------------------------------------------------------------------------------------------
    DisplayEnemyInRoomMessage(Agent: agent): void =
        #HUDEnemyOutRoomCustomMessageDevice.Hide(Agent)
        HUDEnemyInRoomCustomMessageDevice.Show(Agent)
        AudioManagerDevice.PlayAudio_EnemyEntered(Agent)
    #-------------------------------------------------------------------------------------------------
    DisplayRoomClashMessage(Agent: agent): void =
        AudioManagerDevice.PlayAudio_RoomClash(Agent)
    #-------------------------------------------------------------------------------------------------
    DisplayEnemyOutRoomMessage(Agent: agent): void =
        HUDEnemyInRoomCustomMessageDevice.Hide(Agent)
        AudioManagerDevice.PlayAudio_EnemyLeft(Agent)
        #HUDEnemyOutRoomCustomMessageDevice.Show(Agent)
    #-------------------------------------------------------------------------------------------------
    DisplayBlueTeamSwipe(): void =
        Players := GetPlayspace().GetPlayers()
            for(Player:Players):
                HUD_BLUE_SWIPE_MessageDevice.Show(Player)
    #-------------------------------------------------------------------------------------------------
    DisplayRedTeamSwipe(): void =
        Players := GetPlayspace().GetPlayers()
            for(Player:Players):
                HUD_RED_SWIPE_MessageDevice.Show(Player)
    #-------------------------------------------------------------------------------------------------
    OnPlayerElimination(Result : elimination_result): void =                
        if:
            EliminatedAgent := Result.EliminatedCharacter.GetAgent[]             
        then:
            HUDEnemyInRoomCustomMessageDevice.Hide(EliminatedAgent)  
            #Wasted          
            if (IsWastedEnabled = true):                
                HUDWastedCustomMessageDevice.Show(EliminatedAgent)
            if: #CHECK FOR KNOCKOUT
                Instigator := Result.EliminatingCharacter?
                InstigatorAgent := Instigator.GetAgent[]
                InstigatorPlayer := player[InstigatorAgent]
            then:                
                AgentTeam:= GetAgentTeam(EliminatedAgent)
                InstigatorTeam:= GetAgentTeam(InstigatorPlayer)
                if (AgentTeam <> InstigatorTeam):
                    AudioManagerDevice.PlayAudio_PlayerKilledRival(EliminatedAgent)
                    if (AgentTeam = team_id.Blue): #RED KNOCKOUT
                        AudioManagerDevice.PlayAudio_RedKO_1(EliminatedAgent)
                        Players := GetPlayspace().GetPlayers()
                        for(Player:Players):
                            HUD_RED_KO_MessageDevice.Show(Player)
                    else: #BLUE KNOCKOUT
                        AudioManagerDevice.PlayAudio_BlueKO_1(EliminatedAgent)
                        Players := GetPlayspace().GetPlayers()
                        for(Player:Players):
                            HUD_BLUE_KO_MessageDevice.Show(Player)
    #-------------------------------------------------------------------------------------------------
    OnPlayerRespawn(Agent : agent) : void =
        set IsWastedEnabled = true
        # Trigger Spend Coin MSG
        HUDRespawnMessageMessageDevice.Show(Agent)
        # Clear Wasted & other UI
        HUDWastedCustomMessageDevice.Hide(Agent)
        #Objective Widget
        spawn:
            SpawnPlayerObjective(Agent)
    
    SpawnPlayerObjective(Agent : agent) <suspends>: void =
        Sleep(10.0)
        HUDDestroyBedsMessageMessageDevice.Show(Agent)
    
    ClearHUDForAllPlayers() : void =
        Players := GetPlayspace().GetPlayers()
        for(Agent : Players):
            HUDWastedCustomMessageDevice.Hide(Agent)
            HUDBlueBedUnderAttackMessageDevice.Hide(Agent)
            HUDRedBedUnderAttackMessageDevice.Hide(Agent)
            HUD_RED_KO_MessageDevice.Hide(Agent)
            HUD_BLUE_KO_MessageDevice.Hide(Agent)
            HUDEnemyInRoomCustomMessageDevice.Hide(Agent)
            
    #---------------------------------------------------------------------------
    SubscribePlayerToEvents(Player : player): void =
        if:
            FortChar := Player.GetFortCharacter[]
        then:
            FortChar.EliminatedEvent().Subscribe(OnPlayerElimination)
    
    SubscribeSpawners() : void =
        CreativeObjects := FindCreativeObjectsWithTag(SpawnerTag)
        for(SpawnerObject : CreativeObjects):            
            if(Spawner := player_spawner_device[SpawnerObject] ) :
                Spawner.SpawnedEvent.Subscribe(OnPlayerRespawn)
