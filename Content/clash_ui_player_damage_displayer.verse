using { /Fortnite.com/Devices }
using { /Fortnite.com/UI}
using { /Verse.org/Simulation }
using { /Verse.org/Colors }
using { /Verse.org/Colors/NamedColors }
using { /UnrealEngine.com/Temporary/UI }
using { /UnrealEngine.com/Temporary/SpatialMath }

clash_ui_player_damage_displayer := class(clash_ui):

    #------------------------------------------------------------------------------------------------------------
    CreateUI<override>():canvas_slot =
        set UICanvas = canvas_slot:
            Anchors := anchors{Minimum := vector2{X := 0.28, Y := 0.27}, Maximum := vector2{X := 0.48, Y := 0.50}}
            Alignment := vector2{X := 0.5, Y := 0.5}
            Widget:=overlay:
                Slots:=array:
                    overlay_slot:
                        VerticalAlignment:=vertical_alignment.Top
                        HorizontalAlignment:=horizontal_alignment.Center
                        Padding:=margin{Left:=0.0, Top:=0.0, Right:=0.0, Bottom:=0.0}
                        Widget:=texture_block{DefaultImage:=Textures.General.T_GenSquTrans, DefaultDesiredSize:=vector2{X:=1810.0, Y:=170.0}}
                
                    overlay_slot:
                        VerticalAlignment:=vertical_alignment.Top
                        HorizontalAlignment:=horizontal_alignment.Center
                        Padding:=margin{Left:=0.0, Top:=0.0, Right:=0.0, Bottom:=0.0}
                        Widget := ConstructHuntMessage()
        
        return UICanvas

    #------------------------------------------------------------------------------------------------------------
    ConstructHuntMessage(): stack_box =

        HuntMessageStackBox := stack_box:
            Orientation:=orientation.Horizontal
            Slots:=array:
                #First line
                stack_box_slot:
                    VerticalAlignment:=vertical_alignment.Center
                    HorizontalAlignment:=horizontal_alignment.Center
                    #Distribution:=option{0.5}
                    Padding:=margin{Left:=0.0, Top:=0.0, Right:=0.0, Bottom:=10.0}
                    Widget:=overlay:
                        Slots:=array:

                            #Hunting warning icon BG1
                            overlay_slot:
                                VerticalAlignment:=vertical_alignment.Center
                                HorizontalAlignment:=horizontal_alignment.Center
                                #Distribution:=option{0.5}
                                Padding:=margin{Left:=0.0, Top:=0.0, Right:=0.0, Bottom:=0.0}
                                Widget:=texture_block{DefaultImage:=Textures.General.T_StarFrame, DefaultDesiredSize:=vector2{X:=180.0, Y:=180.0}, DefaultTint:=MakeColorFromHex("#fea858FF")}

                            #Hunting warning icon BG2
                            overlay_slot:
                                VerticalAlignment:=vertical_alignment.Center
                                HorizontalAlignment:=horizontal_alignment.Center
                                #Distribution:=option{0.5}
                                Padding:=margin{Left:=0.0, Top:=0.0, Right:=0.0, Bottom:=0.0}
                                Widget:=texture_block{DefaultImage:=Textures.General.T_StarFrame, DefaultDesiredSize:=vector2{X:=132.0, Y:=132.0}, DefaultTint:=MakeColorFromHex("#e67e22FF")}

                            #Hunting warning frame
                            overlay_slot:
                                VerticalAlignment:=vertical_alignment.Center
                                HorizontalAlignment:=horizontal_alignment.Center
                                #Distribution:=option{0.5}
                                Padding:=margin{Left:=5.0, Top:=0.0, Right:=0.0, Bottom:=0.0}
                                Widget:=texture_block{DefaultImage:=Textures.General.T_FrameWarn, DefaultDesiredSize:=vector2{X:=80.0, Y:=80.0}, DefaultTint:=MakeColorFromHex("#FFFFFFFF")}

                            #Hunting warning exclamation sign
                            overlay_slot:
                                VerticalAlignment:=vertical_alignment.Center
                                HorizontalAlignment:=horizontal_alignment.Center
                                #Distribution:=option{0.5}
                                Padding:=margin{Left:=5.0, Top:=0.0, Right:=0.0, Bottom:=0.0}
                                Widget:=texture_block{DefaultImage:=Textures.General.T_IconExclamation, DefaultDesiredSize:=vector2{X:=17.0, Y:=49.0}, DefaultTint:=MakeColorFromHex("#e67e22FF")}

                            #Message frame layout
                            overlay_slot:
                                VerticalAlignment:=vertical_alignment.Center
                                HorizontalAlignment:=horizontal_alignment.Center
                                #Distribution:=option{0.5}
                                Padding:=margin{Left:=200.0, Top:=0.0, Right:=0.0, Bottom:=0.0}
                                Widget:=overlay:
                                    Slots:=array:

                                        #Message frame
                                        overlay_slot:
                                            VerticalAlignment:=vertical_alignment.Center
                                            HorizontalAlignment:=horizontal_alignment.Center
                                            #Distribution:=option{0.5}
                                            Padding:=margin{Left:=0.0, Top:=0.0, Right:=0.0, Bottom:=0.0}
                                            Widget:=texture_block{DefaultImage:=Textures.General.T_MsgFrame, DefaultDesiredSize:=vector2{X:=300.0, Y:=70.0}, DefaultTint:=MakeColorFromHex("#FFFFFFFF")}
                                        
                                        #Actual Message
                                        overlay_slot:
                                            VerticalAlignment:=vertical_alignment.Center
                                            HorizontalAlignment:=horizontal_alignment.Center
                                            #Distribution:=option{0.5}
                                            Padding:=margin{Left:=0.0, Top:=0.0, Right:=0.0, Bottom:=0.0}
                                            Widget:=overlay:
                                                Slots:=array:

                                                    #First line of text
                                                    overlay_slot:
                                                        VerticalAlignment:=vertical_alignment.Center
                                                        HorizontalAlignment:=horizontal_alignment.Center
                                                        #Distribution:=option{0.5}
                                                        Padding:=margin{Left:=0.0, Top:=0.0, Right:=0.0, Bottom:=15.0}
                                                        Widget:=text_block{DefaultText := StringToMessage("An enemy player"), DefaultTextColor:=MakeColorFromHex("#000000FF"), DefaultShadowColor:=MakeColorFromHex("#FF0000FF"), DefaultShadowOffset:=option{vector2{X:=3.0, Y:=3.0}}}

                                                    #Second line of text
                                                    overlay_slot:
                                                        VerticalAlignment:=vertical_alignment.Center
                                                        HorizontalAlignment:=horizontal_alignment.Center
                                                        #Distribution:=option{0.5}
                                                        Padding:=margin{Left:=0.0, Top:=15.0, Right:=0.0, Bottom:=0.0}
                                                        Widget:=text_block{DefaultText := StringToMessage("Hit you!"), DefaultTextColor:=MakeColorFromHex("#FF0000FF"), DefaultShadowColor:=MakeColorFromHex("#000000FF"), DefaultShadowOffset:=option{vector2{X:=3.0, Y:=3.0}}}


        return HuntMessageStackBox
    #------------------------------------------------------------------------------------------------------------

            # #Main structure for the victory message
            # HuntMsg:=stack_box:
            #     Orientation:=orientation.Horizontal
            #                 Slots:=array:

            #                     #First line
            #                     stack_box_slot:
            #                         VerticalAlignment:=vertical_alignment.Center
            #                         HorizontalAlignment:=horizontal_alignment.Center
            #                         #Distribution:=option{0.5}
            #                         Padding:=margin{Left:=0.0, Top:=0.0, Right:=0.0, Bottom:=10.0}
            #                         Widget:=overlay:
            #                             Slots:=array:

            #                                 #Hunting warning icon BG1
            #                                 overlay_slot:
            #                                     VerticalAlignment:=vertical_alignment.Center
            #                                     HorizontalAlignment:=horizontal_alignment.Center
            #                                     #Distribution:=option{0.5}
            #                                     Padding:=margin{Left:=0.0, Top:=0.0, Right:=0.0, Bottom:=0.0}
            #                                     Widget:=texture_block{DefaultImage:=Textures.General.T_StarFrame, DefaultDesiredSize:=vector2{X:=180.0, Y:=180.0}, DefaultTint:=MakeColorFromHex("#fea858FF")}

            #                                 #Hunting warning icon BG2
            #                                 overlay_slot:
            #                                     VerticalAlignment:=vertical_alignment.Center
            #                                     HorizontalAlignment:=horizontal_alignment.Center
            #                                     #Distribution:=option{0.5}
            #                                     Padding:=margin{Left:=0.0, Top:=0.0, Right:=0.0, Bottom:=0.0}
            #                                     Widget:=texture_block{DefaultImage:=Textures.General.T_StarFrame, DefaultDesiredSize:=vector2{X:=132.0, Y:=132.0}, DefaultTint:=MakeColorFromHex("#e67e22FF")}

            #                                 #Hunting warning frame
            #                                 overlay_slot:
            #                                     VerticalAlignment:=vertical_alignment.Center
            #                                     HorizontalAlignment:=horizontal_alignment.Center
            #                                     #Distribution:=option{0.5}
            #                                     Padding:=margin{Left:=5.0, Top:=0.0, Right:=0.0, Bottom:=0.0}
            #                                     Widget:=texture_block{DefaultImage:=Textures.General.T_FrameWarn, DefaultDesiredSize:=vector2{X:=80.0, Y:=80.0}, DefaultTint:=MakeColorFromHex("#FFFFFFFF")}

            #                                 #Hunting warning exclamation sign
            #                                 overlay_slot:
            #                                     VerticalAlignment:=vertical_alignment.Center
            #                                     HorizontalAlignment:=horizontal_alignment.Center
            #                                     #Distribution:=option{0.5}
            #                                     Padding:=margin{Left:=5.0, Top:=0.0, Right:=0.0, Bottom:=0.0}
            #                                     Widget:=texture_block{DefaultImage:=Textures.General.T_IconExclamation, DefaultDesiredSize:=vector2{X:=17.0, Y:=49.0}, DefaultTint:=MakeColorFromHex("#e67e22FF")}

            #                                 #Message frame layout
            #                                 overlay_slot:
            #                                     VerticalAlignment:=vertical_alignment.Center
            #                                     HorizontalAlignment:=horizontal_alignment.Center
            #                                     #Distribution:=option{0.5}
            #                                     Padding:=margin{Left:=200.0, Top:=0.0, Right:=0.0, Bottom:=0.0}
            #                                     Widget:=overlay:
            #                                         Slots:=array:

            #                                             #Message frame
            #                                             overlay_slot:
            #                                                 VerticalAlignment:=vertical_alignment.Center
            #                                                 HorizontalAlignment:=horizontal_alignment.Center
            #                                                 #Distribution:=option{0.5}
            #                                                 Padding:=margin{Left:=0.0, Top:=0.0, Right:=0.0, Bottom:=0.0}
            #                                                 Widget:=texture_block{DefaultImage:=Textures.General.T_MsgFrame, DefaultDesiredSize:=vector2{X:=300.0, Y:=70.0}, DefaultTint:=MakeColorFromHex("#FFFFFFFF")}
                                                        
            #                                             #Actual Message
            #                                             overlay_slot:
            #                                                 VerticalAlignment:=vertical_alignment.Center
            #                                                 HorizontalAlignment:=horizontal_alignment.Center
            #                                                 #Distribution:=option{0.5}
            #                                                 Padding:=margin{Left:=0.0, Top:=0.0, Right:=0.0, Bottom:=0.0}
            #                                                 Widget:=overlay:
            #                                                     Slots:=array:

            #                                                         #First line of text
            #                                                         overlay_slot:
            #                                                             VerticalAlignment:=vertical_alignment.Center
            #                                                             HorizontalAlignment:=horizontal_alignment.Center
            #                                                             #Distribution:=option{0.5}
            #                                                             Padding:=margin{Left:=0.0, Top:=0.0, Right:=0.0, Bottom:=15.0}
            #                                                             Widget:=text_block{DefaultText := StringToMessage("An enemy player"), DefaultTextColor:=MakeColorFromHex("#000000FF"), DefaultShadowColor:=MakeColorFromHex("#FF0000FF"), DefaultShadowOffset:=option{vector2{X:=3.0, Y:=3.0}}}

            #                                                         #Second line of text
            #                                                         overlay_slot:
            #                                                             VerticalAlignment:=vertical_alignment.Center
            #                                                             HorizontalAlignment:=horizontal_alignment.Center
            #                                                             #Distribution:=option{0.5}
            #                                                             Padding:=margin{Left:=0.0, Top:=15.0, Right:=0.0, Bottom:=0.0}
            #                                                             Widget:=text_block{DefaultText := StringToMessage("Hit you!"), DefaultTextColor:=MakeColorFromHex("#FF0000FF"), DefaultShadowColor:=MakeColorFromHex("#000000FF"), DefaultShadowOffset:=option{vector2{X:=3.0, Y:=3.0}}}

            