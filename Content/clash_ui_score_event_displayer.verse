
using { /Fortnite.com/Devices }
using { /Verse.org/Simulation }
using { /UnrealEngine.com/Temporary/Diagnostics }
using { /UnrealEngine.com/Temporary/UI }
using { /Fortnite.com/UI }
using { /Verse.org/Colors/NamedColors }
using { /UnrealEngine.com/Temporary/SpatialMath }
using { /Verse.org/Colors }

clash_ui_score_event_displayer := class(clash_ui):

    var BlueScoreImage : texture_block = texture_block{DefaultImage := Textures.Clash.Score.T_blue_scored, DefaultDesiredSize := vector2{X := 1024.0, Y := 512.0}}

    var RedScoreImage : texture_block = texture_block{DefaultImage := Textures.Clash.Score.T_red_scored, DefaultDesiredSize := vector2{X := 1024.0, Y := 512.0}}

    var BlueScore : texture_block = texture_block{DefaultImage := Textures.General.T_0, DefaultDesiredSize := vector2{X := 54.0, Y := 70.0}}

    var RedScore : texture_block = texture_block{DefaultImage := Textures.General.T_0, DefaultDesiredSize := vector2{X := 54.0, Y := 70.0}}

    #------------------------------------------------------------------------------------------------------------
    CreateUI<override>():canvas_slot =
        # Initialization
        BlueScoreImage.SetVisibility(widget_visibility.Hidden)
        RedScoreImage.SetVisibility(widget_visibility.Hidden)

        set UICanvas = canvas_slot:
            Anchors := anchors{Minimum := vector2{X := 0.5, Y := 0.5}, Maximum := vector2{X := 0.5, Y := 0.5}}
            Alignment := vector2{X := 0.5, Y := 0.5}
            SizeToContent := true
            Widget := overlay:
                Slots := array:
                    overlay_slot:
                        VerticalAlignment:=vertical_alignment.Center
                        HorizontalAlignment:=horizontal_alignment.Center
                        Widget := BlueScoreImage
                    overlay_slot:
                        VerticalAlignment:=vertical_alignment.Center
                        HorizontalAlignment:=horizontal_alignment.Center
                        Widget := RedScoreImage
                    overlay_slot:
                        VerticalAlignment:=vertical_alignment.Center
                        HorizontalAlignment:=horizontal_alignment.Center
                        Padding := margin{Left:=160.0, Top:=50.0}
                        Widget := BlueScore
                    overlay_slot:
                        VerticalAlignment:=vertical_alignment.Center
                        HorizontalAlignment:=horizontal_alignment.Center
                        Padding := margin{Top:=50.0, Right := 185.0}
                        Widget := RedScore

        return UICanvas

    #------------------------------------------------------------------------------------------------------------
    UpdateScoringTeam(ScoringTeam : team_id): void =
        case (ScoringTeam):
            team_id.Blue =>
                BlueScoreImage.SetVisibility(widget_visibility.Visible)
                RedScoreImage.SetVisibility(widget_visibility.Hidden)
            team_id.Red =>
                BlueScoreImage.SetVisibility(widget_visibility.Hidden)
                RedScoreImage.SetVisibility(widget_visibility.Visible)
            _=>

    #------------------------------------------------------------------------------------------------------------
    UpdateScoreTracker(ScoringTeam : team_id, NewScore : int): void =
        case (ScoringTeam):
            team_id.Red =>
                case (NewScore):
                    1 => RedScore.SetImage(Textures.General.T_1)
                    2 => RedScore.SetImage(Textures.General.T_2)
                    3 => RedScore.SetImage(Textures.General.T_3)
                    _=> RedScore.SetImage(Textures.General.T_0)
            team_id.Blue =>
                case (NewScore):
                    1 => BlueScore.SetImage(Textures.General.T_1)
                    2 => BlueScore.SetImage(Textures.General.T_2)
                    3 => BlueScore.SetImage(Textures.General.T_3)
                    _=> BlueScore.SetImage(Textures.General.T_0)
            _=>