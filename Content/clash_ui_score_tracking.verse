
using { /Fortnite.com/Devices }
using { /Fortnite.com/UI }
using { /Verse.org/Colors/NamedColors }
using { /Verse.org/Simulation }
using { /UnrealEngine.com/Temporary/Diagnostics }
using { /UnrealEngine.com/Temporary/SpatialMath }
using { /UnrealEngine.com/Temporary/UI }

clash_ui_score_tracking := class:

	var UICanvas : canvas_slot = canvas_slot{Widget:=stack_box{Orientation:=orientation.Vertical}}

	var RedTeamOverlay:overlay = overlay{}

	var BlueTeamOverlay:overlay = overlay{}

	var RedBedStackBox:stack_box = stack_box{Orientation:=orientation.Horizontal}

	var BlueBedStackBox:stack_box = stack_box{Orientation:=orientation.Horizontal}
	
	var AmountOfBedsToShow:int = 0

	var RoomNameStackBox:stack_box = stack_box{Orientation:=orientation.Horizontal}

	var RoomNameText:text_block = text_block{DefaultTextColor := White}

	#-----------------------------------------------------------------------------------------------------------------------
	SetBedsToShow(NewAmount:int):void =
		set AmountOfBedsToShow = NewAmount

	#-----------------------------------------------------------------------------------------------------------------------
	CreateScoreUI():canvas_slot =
		set UICanvas = canvas_slot:
			Anchors:=anchors{Minimum:=vector2{X:=0.5,Y:=0.0},Maximum:=vector2{X:=0.5,Y:=0.0}}
			Alignment:=vector2{X:=0.5, Y:=0.0}
			SizeToContent:=true
			Offsets:=margin{Top:=100.0}
			Widget:=stack_box:
				Orientation:=orientation.Vertical
				Slots:=array:
					stack_box_slot:
						HorizontalAlignment := horizontal_alignment.Center
						VerticalAlignment := vertical_alignment.Fill
						Widget := ConstructRoomIdentifier()
					stack_box_slot:
						HorizontalAlignment := horizontal_alignment.Center
						VerticalAlignment := vertical_alignment.Fill
						Widget := ConstructScoreBoard()
					<#stack_box_slot:
						HorizontalAlignment := horizontal_alignment.Center
						VerticalAlignment := vertical_alignment.Fill
						Widget := ConstructKillScoreBoard()#>
		return UICanvas

	#-----------------------------------------------------------------------------------------------------------------------
	ConstructRoomIdentifier():stack_box=
		# TODO: Implement room title @jhrojas 31/10/24
		set RoomNameStackBox = StackBox:stack_box = stack_box:
			Orientation:=orientation.Horizontal
			Slots:=array:
				stack_box_slot:
					HorizontalAlignment := horizontal_alignment.Fill
					VerticalAlignment := vertical_alignment.Fill
					Widget := texture_block:
						DefaultImage := Textures.Clash.Score.WhiteBed
				stack_box_slot:
					HorizontalAlignment := horizontal_alignment.Fill
					VerticalAlignment := vertical_alignment.Fill
					Widget := RoomNameText <#text_block:
						DefaultTextColor := White
						DefaultText := StringToMessage("Current Room")#>
		return StackBox

	# Bed scoreboard
	#-----------------------------------------------------------------------------------------------------------------------
	ConstructScoreBoard():stack_box=
		StackBox:stack_box = stack_box:
			Orientation:=orientation.Horizontal
			Slots:=array:
				stack_box_slot:
					HorizontalAlignment:=horizontal_alignment.Fill
					VerticalAlignment:=vertical_alignment.Fill
					Widget:=ConstructRedBeds()
				stack_box_slot:
					HorizontalAlignment:=horizontal_alignment.Fill
					VerticalAlignment:=vertical_alignment.Fill
					Widget:=color_block:
						DefaultColor:=Black
						DefaultDesiredSize:=vector2{X:=4.0, Y:= 32.0}
				stack_box_slot:
					HorizontalAlignment:=horizontal_alignment.Fill
					VerticalAlignment:=vertical_alignment.Fill
					Widget:=ConstructBlueBeds()
		return StackBox

	#-----------------------------------------------------------------------------------------------------------------------
	ConstructRedBeds():stack_box=
		set RedBedStackBox = stack_box:
			Orientation:=orientation.Horizontal
			Slots:=array:
				stack_box_slot:
					HorizontalAlignment:=horizontal_alignment.Fill
					VerticalAlignment:=vertical_alignment.Fill
					Widget:=RedBedWidget()
				stack_box_slot:
					HorizontalAlignment:=horizontal_alignment.Fill
					VerticalAlignment:=vertical_alignment.Fill
					Widget:=RedBedWidget()
				stack_box_slot:
					HorizontalAlignment:=horizontal_alignment.Fill
					VerticalAlignment:=vertical_alignment.Fill
					Widget:=RedBedWidget()
				<#stack_box_slot:
					HorizontalAlignment:=horizontal_alignment.Fill
					VerticalAlignment:=vertical_alignment.Fill
					Widget:=RedBedWidget()
				stack_box_slot:
					HorizontalAlignment:=horizontal_alignment.Fill
					VerticalAlignment:=vertical_alignment.Fill
					Widget:=RedBedWidget()#>
					
		return RedBedStackBox
	
	#-----------------------------------------------------------------------------------------------------------------------
	RedBedWidget():overlay=
		OverlayWidget:overlay = overlay:
			Slots:=array:
				overlay_slot:
					Widget := texture_block{DefaultImage := Textures.Clash.Score.T_Bed_Neutral}
					HorizontalAlignment := horizontal_alignment.Fill
					VerticalAlignment := vertical_alignment.Fill
				overlay_slot:
					Widget := texture_block{DefaultImage := Textures.Clash.Score.T_Bed_red}
					HorizontalAlignment := horizontal_alignment.Fill
					VerticalAlignment := vertical_alignment.Fill

	#-----------------------------------------------------------------------------------------------------------------------
	ConstructBlueBeds():stack_box=
		set BlueBedStackBox = stack_box:
			Orientation:=orientation.Horizontal
			Slots:=array:
				stack_box_slot:
					HorizontalAlignment:=horizontal_alignment.Fill
					VerticalAlignment:=vertical_alignment.Fill
					Widget:=BlueBedWidget()
				stack_box_slot:
					HorizontalAlignment:=horizontal_alignment.Fill
					VerticalAlignment:=vertical_alignment.Fill
					Widget:=BlueBedWidget()
				stack_box_slot:
					HorizontalAlignment:=horizontal_alignment.Fill
					VerticalAlignment:=vertical_alignment.Fill
					Widget:=BlueBedWidget()
				<#stack_box_slot:
					HorizontalAlignment:=horizontal_alignment.Fill
					VerticalAlignment:=vertical_alignment.Fill
					Widget:=BlueBedWidget()
				stack_box_slot:
					HorizontalAlignment:=horizontal_alignment.Fill
					VerticalAlignment:=vertical_alignment.Fill
					Widget:=BlueBedWidget()#>						
			
		return BlueBedStackBox
	
	#-----------------------------------------------------------------------------------------------------------------------
	BlueBedWidget():overlay=
		OverlayWidget:overlay = overlay:
			Slots:=array:
				overlay_slot:
					Widget := texture_block{DefaultImage := Textures.Clash.Score.T_Bed_Neutral}
					HorizontalAlignment := horizontal_alignment.Fill
					VerticalAlignment := vertical_alignment.Fill
				overlay_slot:
					Widget := texture_block{DefaultImage := Textures.Clash.Score.T_Bed_Blue}
					HorizontalAlignment := horizontal_alignment.Fill
					VerticalAlignment := vertical_alignment.Fill
	
	# Player kills scoreboard
	#-----------------------------------------------------------------------------------------------------------------------
	ConstructKillScoreBoard():stack_box=
		StackBox:stack_box = stack_box:
			Orientation:=orientation.Horizontal
			Slots:=array:
				stack_box_slot:
					HorizontalAlignment:=horizontal_alignment.Fill
					VerticalAlignment:=vertical_alignment.Fill
					Widget:=RedTeamKills()
				stack_box_slot:
					HorizontalAlignment:=horizontal_alignment.Fill
					VerticalAlignment:=vertical_alignment.Fill
					Widget:=texture_block{DefaultImage := Textures.Clash.Score.T_Elimination_Icon}
				stack_box_slot:
					HorizontalAlignment:=horizontal_alignment.Fill
					VerticalAlignment:=vertical_alignment.Fill
					Widget:=BlueTeamKills()
		return StackBox
	
	#-----------------------------------------------------------------------------------------------------------------------
	BlueTeamKills():overlay=
		set BlueTeamOverlay = overlay:
			Slots:=array:
				overlay_slot:
					Widget := color_block{DefaultColor := DarkBlue}
					HorizontalAlignment := horizontal_alignment.Fill
					VerticalAlignment := vertical_alignment.Fill
				overlay_slot:
					Widget := text_block:
						DefaultTextColor := White
						DefaultText := StringToMessage("0")
		return BlueTeamOverlay
	
	#-----------------------------------------------------------------------------------------------------------------------
	RedTeamKills():overlay=
		set RedTeamOverlay = overlay:
			Slots:=array:
				overlay_slot:
					Widget := color_block{DefaultColor := DarkRed}
					HorizontalAlignment := horizontal_alignment.Fill
					VerticalAlignment := vertical_alignment.Fill
				overlay_slot:
					Widget := text_block:
						DefaultTextColor := White
						DefaultText := StringToMessage("0")
		return RedTeamOverlay

	# Update Score Values
	#-----------------------------------------------------------------------------------------------------------------------
	UpdateDestroyedBeds(DestroyerTeam:team_id, AmountDestroyed:int):void=
		var BedToRemoveIndex:int = 0
		var StackBoxToUpdate:stack_box = stack_box{Orientation:=orientation.Horizontal}
		case(DestroyerTeam):
			team_id.Blue =>
				set StackBoxToUpdate = BlueBedStackBox
				set BedToRemoveIndex = AmountOfBedsToShow - AmountDestroyed
			team_id.Red =>
				set StackBoxToUpdate = RedBedStackBox
				set BedToRemoveIndex = AmountDestroyed - 1
			_ =>
				Print("Invalid Team")

		SlotsToUpdate := StackBoxToUpdate.Slots
		if:
			WidgetToUpdate := SlotsToUpdate[BedToRemoveIndex].Widget
			OverlayToUpdate := overlay[WidgetToUpdate]
			BedIconSlot := OverlayToUpdate.Slots[1].Widget
		then:
			BedIconSlot.SetVisibility(widget_visibility.Hidden)

	#-----------------------------------------------------------------------------------------------------------------------
	UpdatePlayerKills(KillerTeam:team_id, PlayersKilled:int):void= # TODO JJ: check here is how we udpate ui
		var OverlayToUpdate:overlay = overlay{}
		case(KillerTeam):
			team_id.Blue =>
				set OverlayToUpdate = BlueTeamOverlay
			team_id.Red =>
				set OverlayToUpdate = RedTeamOverlay
			_ =>
				Print("Invalid Team")

		if:
			TextSlot:=OverlayToUpdate.Slots[1]
			TextToUpdate:=text_block[TextSlot.Widget]
		then:
			TextToUpdate.SetText(StringToMessage("{PlayersKilled}"))

	#-----------------------------------------------------------------------------------------------------------------------
	UpdateRoomName(Player:player, RoomName:string):void=
		Print("--- NAME TIME BABEH ! {RoomName} ---")
		RoomNameText.SetText(StringToMessage("{RoomName}"))
		
