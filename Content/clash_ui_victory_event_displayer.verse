using { /Fortnite.com/Devices }
using { /Fortnite.com/UI}
using { /Verse.org/Simulation }
using { /Verse.org/Colors }
using { /Verse.org/Colors/NamedColors }
using { /UnrealEngine.com/Temporary/UI }
using { /UnrealEngine.com/Temporary/SpatialMath }

clash_ui_victory_event_displayer := class(clash_ui):

    RedTitleDefeat : texture_block = texture_block{DefaultImage := Textures.Clash.Score.T_red_defeat_title, DefaultDesiredSize := vector2{X := 1024.0, Y := 512.0}}
    
    RedTitleVictory : texture_block = texture_block{DefaultImage := Textures.Clash.Score.T_red_victory_title, DefaultDesiredSize := vector2{X := 1024.0, Y := 512.0}}

    RedBannerDefeat : texture_block = texture_block{DefaultImage := Textures.Clash.Score.T_red_defeat_banner, DefaultDesiredSize := vector2{X := 2048.0, Y := 400.0}}

    RedBannerVictory : texture_block = texture_block{DefaultImage := Textures.Clash.Score.T_red_victory_banner, DefaultDesiredSize := vector2{X := 2048.0, Y := 510.0}}
    
    BlueTitleDefeat : texture_block = texture_block{DefaultImage := Textures.Clash.Score.T_blue_defeat_title, DefaultDesiredSize := vector2{X := 1024.0, Y := 512.0}}
    
    BlueTitleVictory : texture_block = texture_block{DefaultImage := Textures.Clash.Score.T_blue_victory_title, DefaultDesiredSize := vector2{X := 1024.0, Y := 512.0}}

    BlueBannerDefeat : texture_block = texture_block{DefaultImage := Textures.Clash.Score.T_blue_defeat_banner, DefaultDesiredSize := vector2{X := 2048.0, Y := 400.0}}

    BlueBannerVictory : texture_block = texture_block{DefaultImage := Textures.Clash.Score.T_blue_victory_banner, DefaultDesiredSize := vector2{X := 2048.0, Y := 510.0}}

    var BlueScore : texture_block = texture_block{DefaultImage := Textures.General.T_0, DefaultDesiredSize := vector2{X := 84.0, Y := 100.0}}

    var RedScore : texture_block = texture_block{DefaultImage := Textures.General.T_0, DefaultDesiredSize := vector2{X := 84.0, Y := 100.0}}

    #------------------------------------------------------------------------------------------------------------
    CreateUI<override>():canvas_slot =
        RedTitleDefeat.SetVisibility(widget_visibility.Hidden)
        RedTitleVictory.SetVisibility(widget_visibility.Hidden)
        RedBannerDefeat.SetVisibility(widget_visibility.Hidden)
        RedBannerVictory.SetVisibility(widget_visibility.Hidden)
        BlueTitleDefeat.SetVisibility(widget_visibility.Hidden)
        BlueTitleVictory.SetVisibility(widget_visibility.Hidden)
        BlueBannerDefeat.SetVisibility(widget_visibility.Hidden)
        BlueBannerVictory.SetVisibility(widget_visibility.Hidden)
        BlueScore.SetVisibility(widget_visibility.Hidden)
        RedScore.SetVisibility(widget_visibility.Hidden)

        set UICanvas = canvas_slot:
            Anchors := anchors{Minimum := vector2{X := 0.5, Y := 0.5}, Maximum := vector2{X := 0.5, Y := 0.5}}
            Alignment := vector2{X := 0.5, Y := 0.5}
            SizeToContent := true
            Widget := overlay:
                Slots := array:
                    overlay_slot:
                        VerticalAlignment:=vertical_alignment.Center
                        HorizontalAlignment:=horizontal_alignment.Center
                        Padding:=margin{Bottom:=260.0}
                        Widget:=RedTitleDefeat
                    overlay_slot:
                        VerticalAlignment:=vertical_alignment.Center
                        HorizontalAlignment:=horizontal_alignment.Center
                        Padding:=margin{Bottom:=260.0}
                        Widget:=RedTitleVictory
                    overlay_slot:
                        VerticalAlignment:=vertical_alignment.Center
                        HorizontalAlignment:=horizontal_alignment.Center
                        Padding:=margin{Top:=140.0}
                        Widget:=RedBannerDefeat
                    overlay_slot:
                        VerticalAlignment:=vertical_alignment.Center
                        HorizontalAlignment:=horizontal_alignment.Center
                        Padding:=margin{Top:=100.0}
                        Widget:=RedBannerVictory
                    overlay_slot:
                        VerticalAlignment:=vertical_alignment.Center
                        HorizontalAlignment:=horizontal_alignment.Center
                        Padding:=margin{Bottom:=260.0}
                        Widget:=BlueTitleDefeat
                    overlay_slot:
                        VerticalAlignment:=vertical_alignment.Center
                        HorizontalAlignment:=horizontal_alignment.Center
                        Padding:=margin{Bottom:=260.0}
                        Widget:=BlueTitleVictory
                    overlay_slot:
                        VerticalAlignment:=vertical_alignment.Center
                        HorizontalAlignment:=horizontal_alignment.Center
                        Padding:=margin{Top:=140.0}
                        Widget:=BlueBannerDefeat
                    overlay_slot:
                        VerticalAlignment:=vertical_alignment.Center
                        HorizontalAlignment:=horizontal_alignment.Center
                        Padding:=margin{Top:=100.0}
                        Widget:=BlueBannerVictory
                    overlay_slot:
                        VerticalAlignment:=vertical_alignment.Center
                        HorizontalAlignment:=horizontal_alignment.Center
                        Padding := margin{Left:=210.0, Top:=80.0}
                        Widget := BlueScore
                    overlay_slot:
                        VerticalAlignment:=vertical_alignment.Center
                        HorizontalAlignment:=horizontal_alignment.Center
                        Padding := margin{Top:=80.0, Right := 210.0}
                        Widget := RedScore

        return UICanvas

    #------------------------------------------------------------------------------------------------------------
    UpdateVictoryTeam(AgentTeam : team_id, WinningTeam : team_id): void =
        if (WinningTeam = team_id.Red):
            if (AgentTeam = WinningTeam):
                RedTitleVictory.SetVisibility(widget_visibility.Visible)
                RedBannerVictory.SetVisibility(widget_visibility.Visible)
            else:
                BlueTitleDefeat.SetVisibility(widget_visibility.Visible)
                BlueBannerDefeat.SetVisibility(widget_visibility.Visible)
        else:
            if (AgentTeam = WinningTeam):
                BlueTitleVictory.SetVisibility(widget_visibility.Visible)
                BlueBannerVictory.SetVisibility(widget_visibility.Visible)
            else:
                RedTitleDefeat.SetVisibility(widget_visibility.Visible)
                RedBannerDefeat.SetVisibility(widget_visibility.Visible)

    #------------------------------------------------------------------------------------------------------------
    UpdateScoreTracker(ScoringTeam : team_id, NewScore : int): void =
        RedScore.SetVisibility(widget_visibility.Visible)
        BlueScore.SetVisibility(widget_visibility.Visible)
        case (ScoringTeam):
            team_id.Red =>
                case (NewScore):
                    1 => RedScore.SetImage(Textures.General.T_1)
                    2 => RedScore.SetImage(Textures.General.T_2)
                    3 => RedScore.SetImage(Textures.General.T_3)
                    _=> RedScore.SetImage(Textures.General.T_0)
            team_id.Blue =>
                case (NewScore):
                    1 => BlueScore.SetImage(Textures.General.T_1)
                    2 => BlueScore.SetImage(Textures.General.T_2)
                    3 => BlueScore.SetImage(Textures.General.T_3)
                    _=> BlueScore.SetImage(Textures.General.T_0)
            _=>