<# ... Add here the subfolders so they become public ...#>
#Classes
VerseClasses<public> := module:
    UI<public> := module:
        Stores<public> := module{}

Textures<public> := module:
    Boosters<public> := module{}
    Currency<public> := module{}
    Enemies<public> := module{}
    General<public> := module{}
    #-----------ADDED------------
    RewardThumbnail<public> := module:
    RewardBackgrounds<public> := module:
    ItemIcons<public> := module:        
        CurrencyIcons<public> := module{}     
        ExplosiveIcons<public> := module{}
        DMRIcons<public> := module{}        
        PistolIcons<public> := module{}
        RifleIcons<public> := module{}
        ShotgunIcons<public> := module{}
        SMGIcons<public> := module{}
        SniperIcons<public> := module{}
        XPIcons<public> := module{}
        ConsumableIcons<public> := module:
            Healing<public> := module{}
            Offensives<public> := module{}
        OthersIcons<public> := module:
            Melee<public> := module{}
            Bows<public> := module{}
    #---------------------------------------
    Items<public> := module:
        Artifacts<public> := module {}
        Consumables<public> := module {}
        DMR_Snipers<public> := module {}
        Explosives<public> := module {}
        Pistols<public> := module {}
        Rifles<public> := module {}
        Shotguns<public> := module {}
        SMGs<public> := module {}
        #...
    Powerups<public> := module{}
    Rarity<public> := module{}
    Rescue<public> := module{}
    Rewards<public> := module{}
    Prestige<public> := module{}
    
    Clash<public> := module:
        Score<public> := module {}

Materials<public> := module:
    Crown<public> := module{}
    Identifiers<public> := module{}

    #...