
using { /Fortnite.com/Characters }
using { /Fortnite.com/Devices }
using { /Fortnite.com/Game }
using { /Verse.org/Simulation }
using { /Verse.org/Simulation/Tags }
using { /UnrealEngine.com/Temporary/Diagnostics }
using { /UnrealEngine.com/Temporary/UI }
using { /Fortnite.com/Characters }
using { /Fortnite.com/Game }
using { /Verse.org/Simulation }
using { /Verse.org/Simulation/Tags }
using { /UnrealEngine.com/Temporary/Diagnostics }

log_enemies_manager := class(log_channel){}

# Note: In Editor Stages and Enemy Levels start from 1, in code they are Zero Based Index.

# Tags to identify Spawners by Stage
GuardSpawner_S1_Tag := class(tag){}
GuardSpawner_S2_Tag := class(tag){}
GuardSpawner_S3_Tag := class(tag){}
GuardSpawner_S4_Tag := class(tag){}
GuardSpawner_S5_Tag := class(tag){}
GuardSpawner_S6_Tag := class(tag){}
GuardSpawner_S7_Tag := class(tag){}
GuardSpawner_S8_Tag := class(tag){}
GuardSpawner_S9_Tag := class(tag){}
GuardSpawner_S10_Tag := class(tag){}
GuardSpawner_S11_Tag := class(tag){}
GuardSpawner_S12_Tag := class(tag){}
GuardSpawner_S13_Tag := class(tag){}
GuardSpawner_S14_Tag := class(tag){}
GuardSpawner_S15_Tag := class(tag){}
GuardSpawner_S16_Tag := class(tag){}
GuardSpawner_S17_Tag := class(tag){}
GuardSpawner_S18_Tag := class(tag){}
GuardSpawner_S19_Tag := class(tag){}
GuardSpawner_S20_Tag := class(tag){}
GuardSpawner_S21_Tag := class(tag){}
GuardSpawner_S22_Tag := class(tag){}
GuardSpawner_S23_Tag := class(tag){}
GuardSpawner_S24_Tag := class(tag){}
GuardSpawner_S25_Tag := class(tag){}

# Enemy Level
Enemy_Level_1_Tag := class(tag){} 
Enemy_Level_2_Tag := class(tag){}
Enemy_Level_3_Tag := class(tag){}
Enemy_Level_4_Tag := class(tag){}   
Enemy_Level_5_Tag := class(tag){} 

# Bosses Level
Boss_Level_1_Tag := class(tag){} 
Boss_Level_2_Tag := class(tag){} 
Boss_Level_3_Tag := class(tag){} 
Boss_Level_4_Tag := class(tag){} 
Boss_Level_5_Tag := class(tag){} 

Empty_Tag := class(tag){}

stage := class:
    # Guard spawners for this stage
    var Spawners:[]guard_spawner_device

    # Keeps reference of the agents spawned and alive in this stage
    var SpawnedAgents:[]agent = array{}

    # Know if the stage is despawning so eliminated events are not sent
    var bIsDespawning:logic = false

    AddSpawnedAgent(Agent:agent): void =
        set SpawnedAgents += array{Agent}

    RemoveSpawnedAgent(Agent:agent): void =
        set SpawnedAgents = SpawnedAgents.RemoveAllElements(Agent)

    ResetSpawnedAgents<protected>(): void =
        set SpawnedAgents = array{}

    GetLivingGuardsCount()<transacts>: int =
        SpawnedAgents.Length

    # Returns the max count of guards this stage will have
    GetMaxGuardsCount()<transacts>:int =
        var MaxGuardsCount:int = 0
        for (Spawner : Spawners):
            set MaxGuardsCount += Spawner.GetSpawnLimit()
        MaxGuardsCount

    # Returns if some guard is dead and current living guards is not the max one
    AreSpawnersDirty()<decides><transacts>:void =
        GetLivingGuardsCount() < GetMaxGuardsCount()

    # Resets spawners if dirty
    InitializeSpawnersIfDirty():void =
        if (AreSpawnersDirty[]):
            Print("Room Manager: Initializing spawners NPCs")
            DespawnAll()

            for (Spawner : Spawners):
                for (Count := 0..(Spawner.GetSpawnLimit()-1)):
                    Spawner.Spawn()

    DisableSpawners():void =
        DespawnAll()

    DespawnAll<protected>(): void = 
        set bIsDespawning = true

        for (Spawner : Spawners):
            Spawner.Despawn()
        
        ResetSpawnedAgents()

        set bIsDespawning = false

stage_interaction_result_payload := struct:
    Interaction:device_ai_interaction_result
    StageIndex:int
    Spawner:guard_spawner_device

GetGuardSpawnerTags():[]tag = {
    GuardSpawner_S1_Tag{},
    GuardSpawner_S2_Tag{},
    GuardSpawner_S3_Tag{},
    GuardSpawner_S4_Tag{},
    GuardSpawner_S5_Tag{},
    GuardSpawner_S6_Tag{},
    GuardSpawner_S7_Tag{},
    GuardSpawner_S8_Tag{},
    GuardSpawner_S9_Tag{},
    GuardSpawner_S10_Tag{},
    GuardSpawner_S11_Tag{},
    GuardSpawner_S12_Tag{},
    GuardSpawner_S13_Tag{},
    GuardSpawner_S14_Tag{},
    GuardSpawner_S15_Tag{},
    GuardSpawner_S16_Tag{},
    GuardSpawner_S17_Tag{},
    GuardSpawner_S18_Tag{},
    GuardSpawner_S19_Tag{},
    GuardSpawner_S20_Tag{},
    GuardSpawner_S21_Tag{},
    GuardSpawner_S22_Tag{},
    GuardSpawner_S23_Tag{},
    GuardSpawner_S24_Tag{},
    GuardSpawner_S25_Tag{}
}
GetEnemyLevelTags():[]tag ={
    Enemy_Level_1_Tag{}, 
    Enemy_Level_2_Tag{},
    Enemy_Level_3_Tag{},
    Enemy_Level_4_Tag{},   
    Enemy_Level_5_Tag{},
    Boss_Level_1_Tag{},
    Boss_Level_2_Tag{},
    Boss_Level_3_Tag{},
    Boss_Level_4_Tag{},
    Boss_Level_5_Tag{}
}

enemies_manager := class(creative_device, stat_sender):
    Logger:log = log{Channel:=log_enemies_manager}

    var GameStages: []stage = array{}

    @editable
    PlayerTempData : player_temp_data = player_temp_data{}

    #@editable
    #CurrencyUpgradesManager:currency_upgrades_manager = currency_upgrades_manager{}

    var StatsManager:?stats_manager = false

    OnStageEmptyEvent<public>:event(int) = event(int){}

    OnStageNotEmptyEvent<public>:event(int) = event(int){}

    EnemyReceivedDamageEvent:event(damage_result) = event(damage_result){}

    EnemyEliminatedEvent:event(stage_interaction_result_payload) = event(stage_interaction_result_payload){}

    OnBegin<override>()<suspends>:void=
        Logger.Print("Getting Spawners by Tag...")

        if(FoundStatsManager := GetStatManager[]):
            set StatsManager = option{FoundStatsManager}

        var DebugStageIndex:int=0
        GuardSpawnerTags:=GetGuardSpawnerTags()
        for (GuardSpawnerTag : GuardSpawnerTags):
            set DebugStageIndex += 1
            NewStage := stage{Spawners := GetWorldSpawners(GuardSpawnerTag)}
            set GameStages += array{NewStage}
            Logger.Print("Stage: {DebugStageIndex}, Spawners Found: {NewStage.Spawners.Length}")
            #PrintStageSpawnersPosition(DebugStageIndex-1)

        Logger.Print("Total Stages: {GameStages.Length}")

        # Bind Enemy Spawn and Dead
        for (GameStageIndex := 0..(GameStages.Length-1)):
            for (Spawner: GameStages[GameStageIndex].Spawners):
                spawn{AwaitGuardSpawnersEvents(GameStageIndex, Spawner)}

    ActivateStage(Index:int) : void =
        if (Index <= INDEX_NONE):
            return

        if (GameStage := GameStages[Index]):
            Logger.Print("Try Initializing spawners NPCs in stage {Index} room {Index+1}")
            GameStage.InitializeSpawnersIfDirty()

    DeactivateStage(Index:int) : void =
        if (Index <= INDEX_NONE):
            return

        if (GameStage := GameStages[Index]):
            Logger.Print("Disabling spawners NPCs in stage {Index} room {Index+1}")
            GameStage.DisableSpawners()

    GetWorldSpawners(Tag: tag) : []guard_spawner_device =
        TaggedActors := FindCreativeObjectsWithTag(Tag)
            var Objects:[]creative_object_interface=array{}
                for (Obj:TaggedActors, TaggedActor:=creative_device_base[Obj]):
                    set Objects += array{TaggedActor}
                    #return Objects
        var FoundSpawners:[]guard_spawner_device = array{}
        
        set FoundSpawners = for:
            TaggedActor : TaggedActors
            SpawnerDevice := guard_spawner_device[TaggedActor]
        do:
            SpawnerDevice
        return FoundSpawners

    AwaitGuardSpawnersEvents(StageIndex:int, Spawner:guard_spawner_device)<suspends>:void =
        sync:
            loop:
                Agent:agent = Spawner.SpawnedEvent.Await()

                if (FortCharacter:=Agent.GetFortCharacter[]):
                    FortCharacter.DamagedEvent().Subscribe(OnGuardReceivedDamage)

                if (GameStage:= GameStages[StageIndex]):
                    LastLivingGuardsCount := GameStage.GetLivingGuardsCount()

                    GameStage.AddSpawnedAgent(Agent)

                    if (LastLivingGuardsCount = 0):
                        OnStageNotEmptyEvent.Signal(StageIndex)
                        
                    Logger.Print("Stage: {StageIndex+1}, Living Guard Added, Total Living Guards: {GameStage.GetLivingGuardsCount()}")

            loop:
                Result:device_ai_interaction_result = Spawner.EliminatedEvent.Await()

                EnemyEliminatedEvent.Signal(stage_interaction_result_payload{Interaction:=Result, StageIndex:=StageIndex, Spawner := Spawner})

                if:
                    GameStage:= GameStages[StageIndex]
                    not GameStage.bIsDespawning?
                    KilledAgent:=Result.Target?
                    GameStage.SpawnedAgents.Find[KilledAgent] # If not added means it was also called after a despawning
                then:
                    GameStage.RemoveSpawnedAgent(KilledAgent)

                    if (GameStage.GetLivingGuardsCount() <= 0):
                        OnStageEmptyEvent.Signal(StageIndex)

                    if (PlayerKiller := player[Result.Source?]):
                        EnemyLevel := GetSpawnerEnemyLevelIndex(Spawner)
                        if (EnemyLevel >= 0):
                            #GiveEliminationCurrencyReward(PlayerKiller, EnemyLevel)
                            
                    Logger.Print("Stage: {StageIndex+1}, Left Guards: {GameStage.GetLivingGuardsCount()}")

    # Gives currency rewards for an elimination
    <#GiveEliminationCurrencyReward(PlayerKiller:player, Level:int): void =
        var KillStreakMultiplier : float = 1.0

        for (Player:GetPlayspace().GetPlayers()):
            if(ValidMultiplier:=PlayerTempData.GetPlayerKillStreakMultiplier[Player], ValidMultiplier > 0.0):
                set KillStreakMultiplier = ValidMultiplier
                Print("KillStreak Multiplier: {KillStreakMultiplier}")

            OI:=optional_iterator{Value:= KillStreakMultiplier, IncrementType:= increment_type.Multiply}

            var Data:[]stat_data = array{}
            var SpecificID:stat_id_new = stat_id_new.None
            var GeneralID:stat_id_new = stat_id_new.None
            case(Level):
                0 => set SpecificID = stat_id_new.Type1EnemiesEliminated
                1 => set SpecificID = stat_id_new.Type2EnemiesEliminated
                2 => set SpecificID = stat_id_new.Type3EnemiesEliminated
                3 => set SpecificID = stat_id_new.Type4EnemiesEliminated
                4 => set SpecificID = stat_id_new.Type5EnemiesEliminated
                5 => set SpecificID = stat_id_new.Type1BossesEliminated
                6 => set SpecificID = stat_id_new.Type2BossesEliminated
                7 => set SpecificID = stat_id_new.Type3BossesEliminated
                8 => set SpecificID = stat_id_new.Type4BossesEliminated
                9 => set SpecificID = stat_id_new.Type5BossesEliminated
                _ => {}

            set GeneralID = if (Level <= 4) then stat_id_new.EnemiesEliminated else stat_id_new.BossesEliminated

            if(not Player = PlayerKiller):
                CurrencyUpgradesManager.GiveCurrencyReward(Player, currency_upgrade_id.Assist, Level)
            else:        
                CurrencyUpgradesManager.GiveCurrencyReward(Player, currency_upgrade_id.Elimination, Level, ?OptionalIterator:=option{OI})

                set Data += array{
                    stat_data{Player := Player, Category := stat_category.General, Type := stat_type_new.Number, ID := SpecificID, Value := 1.0}
                    stat_data{Player := Player, Category := stat_category.Run, Type := stat_type_new.Number, ID := GeneralID, Value := 1.0}
                    stat_data{Player := Player, Category := stat_category.General, Type := stat_type_new.Number, ID := GeneralID, Value := 1.0}
                }
            if:
                SendStats[Data]#>

    GetSpawnerEnemyLevelIndex(Spawner:guard_spawner_device):int =
        SpawnerTags := Spawner.GetTags()
        EnemyLevelTags:=GetEnemyLevelTags()
        for (TagIndex := 0..(EnemyLevelTags.Length-1), Tag := EnemyLevelTags[TagIndex]):
            if (SpawnerTags.Has[Tag]):
                return TagIndex

        return INDEX_NONE

    OnGuardReceivedDamage(Result:damage_result): void =
        EnemyReceivedDamageEvent.Signal(Result)
        
    GetGuardsInStage(Index:int)<transacts>: []agent =
        if (Stage:=GameStages[Index]):
            Stage.SpawnedAgents
        else:
            array{}
        
    PrintStageSpawnersPosition(StageIndex:int):void=
        if (Stage:=GameStages[StageIndex]):
            for (Spawner : Stage.Spawners):
                Transform:= Spawner.GetTransform()
                Logger.Print("Stage: {StageIndex} | Spawner Location: X: {Transform.Translation.X}, Y: {Transform.Translation.Y}, Z: {Transform.Translation.Z}")

    SendStats<override>(Data:[]stat_data)<decides><transacts>:void = 
        if (Manager := StatsManager?):
            Manager.ReceiveData[Data]

    GetStatManager<override>()<decides><transacts>:stats_manager = 
        TaggedObjects := FindCreativeObjectsWithTag(stat_manager_tag{})
            var Objects:[]creative_object_interface=array{}
                for (Obj:TaggedObjects, TaggedObject:=creative_device_base[Obj]):
                    set Objects += array{TaggedObject}
                    #return Objects 
        Managers := for (Obj : TaggedObjects, Manager := stats_manager[Obj]):
            Manager
        if (Result := Managers[0]):
            Result
        else:
            Print("No stats manager found! returning empty instance")
            stats_manager{}
            

