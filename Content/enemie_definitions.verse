using { /Verse.org/Assets }

<# ... Add here the textures for the enemies ...#>
FindEnemyIcon(Tier:int) : texture =
    case(Tier):
        0 => Textures.Enemies.Kavel
        1 => Textures.Enemies.Sanctum
        2 => Textures.Enemies.B_R_U_T_E
        3 => Textures.Enemies.Cyclo
        4 => Textures.Enemies.Taro
        5 => Textures.Enemies.Beastmode
        6 => Textures.Enemies.Bogstick
        7 => Textures.Enemies.Shogun
        8 => Textures.Enemies.Undying_Sorrow
        9 => Textures.Enemies.Bull_Shark
        # ...

        _ => Textures.T_Blank