using { /Fortnite.com/Devices }
using { /Fortnite.com/UI }
using { /Fortnite.com/Characters }
using { /Verse.org/Assets }
using { /Verse.org/Colors }
using { /Verse.org/Colors/NamedColors }
using { /Verse.org/Simulation }
using { /UnrealEngine.com/Temporary/SpatialMath }
using { /UnrealEngine.com/Temporary/UI }
using { /Verse.org/Simulation/Tags }

player_hud_data := class():

	Canvas:canvas

	#CurrencyTextBlocks:weak_map(currency_id, text_block)

	#CurrenciesStackBox:stack_box

	Components:[]stack_box

	# Clash Data
	var MaybeClashHudUI:?ui_hud_match_marker = false
	#var MaybeClashScoreUI:?clash_ui_score_tracking = false
	var MaybeClashScorePointUI:?clash_ui_score_event_displayer = false
	var MaybeClashVictoryUI: ?clash_ui_victory_event_displayer = false
    var MaybeClashDamageUI: ?clash_ui_player_damage_displayer = false
    var MaybeClashPlayerEliminatedUI: ?clash_ui_on_player_eliminated_displayer = false
    var MaybeClashPlayerEliminatedPlayerUI: ?clash_ui_on_player_eliminated_player_displayer = false
	var MaybeClashEncounterUI : ?ui_clash_encounter = false

hud_manager := class(creative_device):
	@editable # If true, the ui manager will be initialized in it's OnBegin method. Otherwise, InitializeManager() needs to be called manually
	bAutoInitialize:logic = true

	#@editable
	#DisplayCurrencies:[]currency_id = array{currency_id.Gold, currency_id.Diamond}

	@editable
	CurrenciesPosition:vector2 = vector2{X:=-10.0,Y:=-260.0}

	@editable
	CurrencyTextColor:color = NamedColors.Black

	@editable
	CurrencySlotSize:vector2 = vector2{X:=300.0,Y:=65.0}

	@editable
	CurrencyIconSize:vector2 = vector2{X:=45.0,Y:=45.0}

	@editable
	RoomManager:clash_room_manager = clash_room_manager{}

	@editable
	CurrencyManager:clash_currency_manager = clash_currency_manager{}

	@editable
	PlayerControllerManager : player_controller_manager = player_controller_manager{}

	@editable
	LobbyManagerTag:tag = lobby_manager_device_tag{}

	@editable
	var LobbyManager : clash_lobby_manager_device = clash_lobby_manager_device{}

	@editable
	ClashPlayerEliminationManager : elimination_manager_device = elimination_manager_device{}

	var bIsInitialized:logic = false

	var PlayersHudData:[player]player_hud_data = map{}

	#-----------------------------------------------------------------------------------------------------------------------
	# Clash values
	@editable
	ClashObjectiveManager:clash_objective_manager = clash_objective_manager{}

    @editable
    ClashDamageTracker:clash_damage_tracker_device = clash_damage_tracker_device{}

	#Interrup event
	OnHUDInterruption:event(int)=event(int){}

	#-----------------------------------------------------------------------------------------------------------------------
	OnBegin<override>()<suspends>:void =

		if (bAutoInitialize?):
			InitializeManager()

		ClashPlayerEliminationManager.EliminatedEvent.Subscribe(HideUIOnDead)

		#sync:
			<#loop:
				CurrencyChangedData := CurrencyManager.CurrencyChangedEvent.Await()
				if (PlayerHudData:=PlayersHudData[CurrencyChangedData.Player]):
					UpdateCurrency(CurrencyChangedData.Player, PlayerHudData, CurrencyChangedData.Currency, CurrencyChangedData.CurrentValue)#>
		
	#-----------------------------------------------------------------------------------------------------------------------
	HideUIOnDead(KilledAgent : agent) : void =
		if:
			DeadPlayer := player[KilledAgent]
			PlayerHudData:= PlayersHudData[DeadPlayer]
			NewClashEncounterHudUI := PlayerHudData.MaybeClashEncounterUI?
		then:
			NewClashEncounterHudUI.SetVisibility(false)

	#-----------------------------------------------------------------------------------------------------------------------
	InitializeManager():void =
		if (not bIsInitialized?):
			for (Player:GetPlayspace().GetPlayers()):
				OnPlayerAdded(Player)
			
			GetPlayspace().PlayerAddedEvent().Subscribe(OnPlayerAdded)
			GetPlayspace().PlayerRemovedEvent().Subscribe(OnPlayerRemoved)
			set bIsInitialized = true
			RoomManager.SetHUDManagerReferenceForRooms(Self)

		if (CheckForLobby()?):
			Print("Lobby!")

	#-----------------------------------------------------------------------------------------------------------------------
	# Check for lobby manager. If present check if Lobby is habilitated
	CheckForLobby()<transacts> : logic =
		CreativeObjects:=FindCreativeObjectsWithTag(LobbyManagerTag)
		for(Creativedevice: CreativeObjects, Lobby := clash_lobby_manager_device[Creativedevice]):
			set LobbyManager = Lobby
			if:
				Lobby.IsLobbyEnabled?
			then:
				
				return true
			else:
				return false
		return false

	#-----------------------------------------------------------------------------------------------------------------------
	OnPlayerAdded(Player:player):void =
		
		if (not PlayersHudData[Player], PlayerUI := GetPlayerUI[Player]):
			PlayerHudData := ConstructPlayerHud(Player)
			if (set PlayersHudData[Player] = PlayerHudData) {}
			PlayerUI.AddWidget(PlayerHudData.Canvas)

			

			# ---------------------------------
			# Upgraded Clash HUD UI
			
			ActiveLobby := CheckForLobby()

			HudUI := ui_hud_match_marker{}
			set PlayerHudData.MaybeClashHudUI = option{HudUI}
			if (NewHudUI := PlayerHudData.MaybeClashHudUI?):
				PlayerHudData.Canvas.AddWidget(NewHudUI.CreateUI())
				spawn{UpdateRooms(HudUI)}
				spawn{UpdateRoomName()}
				spawn{UpdateBedsDestroyed(NewHudUI)}
				spawn{UpdateMatchStarted(NewHudUI)}

			if:
				ActiveLobby?
			then:
				# there is an active lobby, lets hide the HUD for the time being until gameplay starts
				if (MyHudUI := PlayerHudData.MaybeClashHudUI?):
					MyHudUI.SetVisibility(false)

			ClashEncounterUI := ui_clash_encounter{}
			set PlayerHudData.MaybeClashEncounterUI = option{ClashEncounterUI}
			if (NewClashEncounterHudUI := PlayerHudData.MaybeClashEncounterUI?):
				PlayerHudData.Canvas.AddWidget(NewClashEncounterHudUI.CreateUI())
				NewClashEncounterHudUI.SetVisibility(false)

			# ---------------------------------
			# Clash UI # pending to be deprecated

			# ScoreUI := clash_ui_score_tracking{}
			# set PlayerHudData.MaybeClashScoreUI = option{ScoreUI}
			# if (NewScoreUI := PlayerHudData.MaybeClashScoreUI?):
			# 	NewScoreUI.SetBedsToShow(ClashObjectiveManager.ScoreToWin)
			# 	PlayerHudData.Canvas.AddWidget(NewScoreUI.CreateScoreUI())
			# 	spawn{UpdateBedsDestroyed(ScoreUI)}
			# 	spawn{UpdatePlayersKilled(ScoreUI)}

			# ---------------------------------
			# Clash Score UI

			ScorePointUI := clash_ui_score_event_displayer{}
			set PlayerHudData.MaybeClashScorePointUI = option{ScorePointUI}
			if (NewScorePointUI := PlayerHudData.MaybeClashScorePointUI?):
				PlayerHudData.Canvas.AddWidget(NewScorePointUI.CreateUI())
				NewScorePointUI.SetVisibility(false)
				spawn{UpdateScoreTracker(ScorePointUI)}

			# ---------------------------------
			# Clash Victory UI
			VictoryUI := clash_ui_victory_event_displayer{}
			set PlayerHudData.MaybeClashVictoryUI = option{VictoryUI}
			if (NewVictoryUI := PlayerHudData.MaybeClashVictoryUI?):
				PlayerHudData.Canvas.AddWidget(NewVictoryUI.CreateUI())
				NewVictoryUI.SetVisibility(false)
				spawn{UpdateVictory(VictoryUI, Player)}
				spawn{UpdateVictoryTracker(VictoryUI)}

			# ---------------------------------
			# Clash Player Damage
			PlayerDamageUI := clash_ui_player_damage_displayer{}
			set PlayerHudData.MaybeClashDamageUI = option{PlayerDamageUI}
			if (NewPlayerDamageUI := PlayerHudData.MaybeClashDamageUI?):
				PlayerHudData.Canvas.AddWidget(NewPlayerDamageUI.CreateUI())
				NewPlayerDamageUI.SetVisibility(false)
				spawn{DisplayDamage()}

			# ---------------------------------
			<# Clash Player Eliminated By Player
			PlayerEliminatedUI := clash_ui_on_player_eliminated_displayer{}
			set PlayerHudData.MaybeClashPlayerEliminatedUI = option{PlayerEliminatedUI}
			if (NewPlayerEliminatedUI := PlayerHudData.MaybeClashPlayerEliminatedUI?):
				PlayerHudData.Canvas.AddWidget(NewPlayerEliminatedUI.CreateUI())
				NewPlayerEliminatedUI.SetVisibility(false)
				spawn{DisplayElimninatedByPlayer()}

			# ---------------------------------
			# Clash Player Eliminated Another Player
			PlayerEliminatedPlayerUI := clash_ui_on_player_eliminated_player_displayer{}
			set PlayerHudData.MaybeClashPlayerEliminatedPlayerUI = option{PlayerEliminatedPlayerUI}
			if (NewPlayerEliminatedPlayerUI := PlayerHudData.MaybeClashPlayerEliminatedPlayerUI?):
				PlayerHudData.Canvas.AddWidget(NewPlayerEliminatedPlayerUI.CreateUI())
				NewPlayerEliminatedPlayerUI.SetVisibility(false)
				spawn{DisplayEliminatedAnotherPlayer()}#>

	#-----------------------------------------------------------------------------------------------------------------------
	#Clash Functions

	UpdateBedsDestroyed(ClashUIScore:ui_hud_match_marker)<suspends>:void=
		loop:
			BedDestroyedData := ClashObjectiveManager.OnBedDestroyedEvent.Await()
			ClashUIScore.UpdateDestroyedBeds(BedDestroyedData)

	UpdateMatchStarted(ClashUIScore:ui_hud_match_marker)<suspends>:void =
		loop:
			Visible := LobbyManager.OnMatchGameplayStartEvent.Await()
			if (Visible?):
				Print("UpdateMatchStarted :: EVENT RECEIVED ! WITH VALUE TRUE")
			else:
				Print("UpdateMatchStarted :: EVENT RECEIVED ! WITH VALUE FALSE")
			ClashUIScore.SetVisibility(Visible)
		
	# UpdatePlayersKilled(ClashUIScore:clash_ui_score_tracking)<suspends>:void=
	# 	loop:
	# 		BedDestroyedData := ClashObjectiveManager.OnClashPlayerKilledEvent.Await()
	# 		ClashUIScore.UpdatePlayerKills(BedDestroyedData)

	# UpdateRoomName(ClashUIScore:clash_ui_score_tracking)<suspends>:void=
	# 	loop:
	# 		NewRoomData := RoomManager.OnRoomEnteredEvent.Await()
	# 		ClashUIScore.UpdateRoomName(NewRoomData)

	UpdateRooms(ClashHUD:ui_hud_match_marker)<suspends>:void=
		loop:
			NewRoomData := RoomManager.OnRoomsTeamsUpdatedEvent.Await() # PENDING EVENT FOR ROOM UPDATING
			ClashHUD.UpdateRooms(NewRoomData)

	UpdateRoomName()<suspends>: void =
		loop:
			NewRoomData := RoomManager.OnRoomEnteredEvent.Await()
			if:
				CurrentHudData := PlayersHudData[NewRoomData(0)]
				MaybeGeneralHUD := CurrentHudData.MaybeClashHudUI?
			then:
				MaybeGeneralHUD.UpdateRoomName(NewRoomData(1))


	#---------------------------------------------------------
	#Score

	UpdateScoreTracker(ClashUIScorePointer: clash_ui_score_event_displayer)<suspends>: void =
		loop:
			ScoreTrackerData := ClashObjectiveManager.OnBedDestroyedEvent.Await()
			ClashUIScorePointer.UpdateScoringTeam(ScoreTrackerData(0))
			ClashUIScorePointer.UpdateScoreTracker(ScoreTrackerData)
			ClashUIScorePointer.DisplayForSeconds(5.0)

	#---------------------------------------------------------
	#Victory

	UpdateVictory(ClashUIVictory: clash_ui_victory_event_displayer, Agent : agent)<suspends>: void =
		loop:
			VictoryData := ClashObjectiveManager.OnGameFinished.Await()
			ClashUIVictory.UpdateVictoryTeam(GetAgentTeam(Agent), VictoryData)
			ClashUIVictory.DisplayForSeconds(5.0)

	UpdateVictoryTracker(ClashUIVictory: clash_ui_victory_event_displayer)<suspends>: void =
		loop:
			VictoryTrackerData := ClashObjectiveManager.OnUpdatedVictoryTrackerEvent.Await()
			ClashUIVictory.UpdateScoreTracker(VictoryTrackerData)

	#---------------------------------------------------------
    #Damage

	DisplayDamage()<suspends>: void =
		loop:
			Result := ClashDamageTracker.OnPlayerDamagedEvent.Await()
			if:
				Target := Result.Target
				TargetCharacter := fort_character[Target]
				TargetAgent := TargetCharacter.GetAgent[]
				ValidPlayerTarget := player[TargetAgent]
				TargetTeam := GetAgentTeam(ValidPlayerTarget)

				InstigatorCharacter := Result.Instigator?
				InstigatorAgent := InstigatorCharacter.GetInstigatorAgent[]
				ValidInstigatorPlayer := player[InstigatorAgent]
				InstigatorTeam := GetAgentTeam(ValidInstigatorPlayer)

				CurrentHudData := PlayersHudData[ValidPlayerTarget]
				MaybeDamageUI := CurrentHudData.MaybeClashDamageUI?

				InstigatorTeam <> TargetTeam
			then:
				MaybeDamageUI.DisplayForSeconds(2.0)

	#---------------------------------------------------------
	#Eliminated By Player
<#
	DisplayElimninatedByPlayer()<suspends>: void =
		loop:
			Result := ClashDamageTracker.OnPlayerKilledEvent.Await()
			if:
				EliminatedAgent := Result.EliminatedCharacter.GetAgent[]
				EliminatedPlayer := player[EliminatedAgent]

				InstigatorCharacter := Result.EliminatingCharacter?
				InstigatorAgent := InstigatorCharacter.GetAgent[]
				ValidInstigatorPlayer := player[InstigatorAgent]

				EliminatedPlayer <> ValidInstigatorPlayer

				CurrentHudData := PlayersHudData[EliminatedPlayer]
				MaybeEliminatedByPlayerUI := CurrentHudData.MaybeClashPlayerEliminatedUI?
				MaybeDamagedByPlayerUI := CurrentHudData.MaybeClashDamageUI?
			then:
				DamageUIVisible := MaybeDamagedByPlayerUI.GetVisibility()
				if(DamageUIVisible?):
					MaybeDamagedByPlayerUI.SetVisibility(false)
				MaybeEliminatedByPlayerUI.DisplayForSeconds(3.0)

	#---------------------------------------------------------
    #Eliminated Another Player

	DisplayEliminatedAnotherPlayer()<suspends>: void =
		loop:
			Result := ClashDamageTracker.OnPlayerKilledEvent.Await()
			if:
				InstigatorCharacter := Result.EliminatingCharacter?
				InstigatorAgent := InstigatorCharacter.GetAgent[]
				ValidInstigatorPlayer := player[InstigatorAgent]

				EliminatedAgent := Result.EliminatedCharacter.GetAgent[]
				EliminatedPlayer := player[EliminatedAgent]

				EliminatedPlayer <> ValidInstigatorPlayer

				CurrentHudData := PlayersHudData[ValidInstigatorPlayer]
				MaybePlayerEliminatedByPlayerUI := CurrentHudData.MaybeClashPlayerEliminatedPlayerUI?
			then:
				MaybePlayerEliminatedByPlayerUI.DisplayForSeconds(3.0)
#>
#-----------------------------------------------------------------------------------------------------------------------
	OnPlayerRemoved(Player:player):void=
		if (PlayerHudData:=PlayersHudData[Player]):
			if (PlayerUI := GetPlayerUI[Player]):
				PlayerUI.RemoveWidget(PlayerHudData.Canvas)

			var NewMap:[player]player_hud_data = map{}
			for (KeyPlayer->Data : PlayersHudData, KeyPlayer<>Player):
				set NewMap = ConcatenateMaps(NewMap, map{KeyPlayer=>Data})
			set PlayersHudData = NewMap

	#-----------------------------------------------------------------------------------------------------------------------
	ConstructPlayerHud(Player:player):player_hud_data =
		# Currencies
		<#CurrenciesStackBox:stack_box = stack_box{Orientation:=orientation.Vertical}
		var CurrencyTextBlocks:weak_map(currency_id, text_block) = map{}

		for (Currency:DisplayCurrencies):
			TextBlock:=text_block:
				DefaultText:=StringToMessage("{ToNotationString(CurrencyManager.GetPlayerCurrency(Player, Currency))}")
				DefaultTextColor:=CurrencyTextColor

			CurrenciesStackBox.AddWidget(
				stack_box_slot:
					HorizontalAlignment:=horizontal_alignment.Fill
					VerticalAlignment:=vertical_alignment.Fill
					Padding:=margin{Bottom:=5.0}
					Widget:=overlay:
						Slots:=array:
							# Background
							overlay_slot:
								HorizontalAlignment:=horizontal_alignment.Fill
								VerticalAlignment:=vertical_alignment.Fill
								Widget:=texture_block:
									DefaultImage:=FindCurrencySlotBackground(Currency)
									DefaultDesiredSize:=CurrencySlotSize
							# Icon
							overlay_slot:
								HorizontalAlignment:=horizontal_alignment.Left
								VerticalAlignment:=vertical_alignment.Center
								Padding:=margin{Left:=15.0}
								Widget:=texture_block:
									DefaultImage:=FindCurrencyIcon(Currency)
									DefaultDesiredSize:=CurrencyIconSize
							# Text
							overlay_slot:
								HorizontalAlignment:=horizontal_alignment.Fill
								VerticalAlignment:=vertical_alignment.Center
								Padding:=margin{Left:=90.0, Right:=10.0}
								Widget:=TextBlock
			)

			if (set CurrencyTextBlocks[Currency] = TextBlock) {}#>

		# Powerups
		PowerupsStackBox:stack_box = stack_box{Orientation:=orientation.Horizontal}

		Canvas:=canvas:
			Slots:=array:
				<#canvas_slot:
					Anchors:=anchors{Minimum:=vector2{X:=1.0,Y:=1.0},Maximum:=vector2{X:=1.0,Y:=1.0}}
					Offsets:=margin{Top:=CurrenciesPosition.Y, Left:=CurrenciesPosition.X, Right:=0.0, Bottom:=0.0}
					Alignment:=vector2{X:=1.0, Y:=1.0}
					SizeToContent:=true
					Widget:=CurrenciesStackBox#>

		PlayerHudData := player_hud_data:
			Canvas := Canvas

			#CurrencyTextBlocks:=CurrencyTextBlocks
			#CurrenciesStackBox:= CurrenciesStackBox

			#Components:=array{CurrenciesStackBox}
			Components:=array{}

		return PlayerHudData

	<#UpdateCurrency(Player:player, PlayerHudData:player_hud_data, Currency:currency_id, NewValue:float): void =
		if (CurrencyTextBlock:=PlayerHudData.CurrencyTextBlocks[Currency]):
			CurrencyTextBlock.SetText(StringToMessage("{ToNotationString(NewValue)}"))#>

	#-----------------------------------------------------------------------------------------------------------------------
	SetHUDComponentPosition<public>(Player : player, ComponentIndex : int, NewPosition : vector2): void =
		Print("Refreshing HUD...")
		if:
			HudData:= PlayersHudData[Player]
			Component:= HudData.Components[ComponentIndex]
			Print("HUD Data OK")
		then:  
			HudData.Canvas.RemoveWidget(Component)
			HudData.Canvas.AddWidget(
				canvas_slot:
					Anchors:=anchors{Minimum:=vector2{X:=1.0,Y:=1.0},Maximum:=vector2{X:=1.0,Y:=1.0}}
					Offsets:=margin{Top:=NewPosition.Y, Left:=NewPosition.X, Right:=0.0, Bottom:=0.0}
					Alignment:=vector2{X:=1.0, Y:=1.0}
					SizeToContent:=true
					Widget:=Component
			)


	#-----------------------------------------------------------------------------------------------------------------------
    ResetHudSettings<public>(Player : player): void =
        Print("Reseting HUD...")
        SetHUDComponentPosition(Player, 0, CurrenciesPosition)

    #-----------------------------------------------------------------------------------------------------------------------
	DisplayClashEvent(Player : player, OwnerTeam : team_id, RoomName : string) <suspends> : void =
		if (OwnerTeam <> team_id.Neutral):
			return
		if (PlayerHudData:= PlayersHudData[Player]):
			if (NewClashEncounterHudUI := PlayerHudData.MaybeClashEncounterUI?):
				PlayerHudData.Canvas.AddWidget(NewClashEncounterHudUI.CreateUI())
				if (RoomName = "Red Lobby" or RoomName = "Blue Lobby"):
					HideClashEvent(Player)
				else:
					if (NewClashEncounterHudUI.bIsVisible?):
						return
					Sleep(0.2)
					NewClashEncounterHudUI.SetVisibility(true)

	#-----------------------------------------------------------------------------------------------------------------------
	HideClashEvent(Player : player) <suspends> : void =
		if (PlayerHudData:= PlayersHudData[Player]):
			if (NewClashEncounterHudUI := PlayerHudData.MaybeClashEncounterUI?):
				loop:
					NewClashEncounterHudUI.SetVisibility(false)
					if (not NewClashEncounterHudUI.bIsVisible?):
						break

	#-----------------------------------------------------------------------------------------------------------------------
	UpdateClashEventInfo(PlayerInvolved : player, AmountOfEnemies : int, CurrentRoom : string, OwnerTeam : team_id, BlueKills : int, RedKills : int): void =
		if (PlayerHudData:= PlayersHudData[PlayerInvolved]):
			if (NewClashEncounterHudUI := PlayerHudData.MaybeClashEncounterUI?):
				NewClashEncounterHudUI.UpdateSliders(AmountOfEnemies, CurrentRoom, OwnerTeam, BlueKills, RedKills)
	
	#-----------------------------------------------------------------------------------------------------------------------
	UpdateRoomKills(RedTeamKills : int, BlueTeamKills : int, LastKillerAgent : agent, TotalEnemies : int) : void =
		if:
			Player := player[LastKillerAgent]
			PlayerHudData:= PlayersHudData[Player]
			NewClashEncounterHudUI := PlayerHudData.MaybeClashEncounterUI?
		then:
			NewClashEncounterHudUI.UpdateKills(RedTeamKills, BlueTeamKills, TotalEnemies)