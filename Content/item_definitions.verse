using { /Fortnite.com/Devices }
using { /Verse.org/Simulation }
using { /Verse.org/Colors }
using { /Verse.org/Colors/NamedColors }
using { /Verse.org/Assets }

<# 
RARITY
#>
item_rarity := enum{Common, Uncommon, Rare, Epic, Legendary, Mythic, Exotic}

item_rarity_id_array:[]item_rarity = array{
	item_rarity.Common
	item_rarity.Uncommon
	item_rarity.Rare
	item_rarity.Epic
	item_rarity.Legendary
	item_rarity.Mythic  
	item_rarity.Exotic
}

<# ... Add here the textures for the rarity card background ...#>
RarityToMessage(Id:item_rarity) : string =
    case(Id):
        item_rarity.Common => "Common"
        item_rarity.Uncommon => "Uncommon"
        item_rarity.Rare => "Rare"
        item_rarity.Epic => "Epic"
        item_rarity.Legendary => "Legendary"
        item_rarity.Mythic => "Mythic"
        item_rarity.Exotic => "Exotic"
        # ...

RarityToIndex(Id:item_rarity) : int =
	case(Id):
		item_rarity.Common => 0
		item_rarity.Uncommon => 1
		item_rarity.Rare => 2
		item_rarity.Epic => 3
		item_rarity.Legendary => 4
		item_rarity.Mythic => 5
		item_rarity.Exotic => 6
		# ...

IndexToRarity(Index:int) : item_rarity =
	case(Index):
		0 => item_rarity.Common
		1 => item_rarity.Uncommon 
		2 => item_rarity.Rare 
		3 => item_rarity.Epic
		4 => item_rarity.Legendary
		5 => item_rarity.Mythic 
		6 => item_rarity.Exotic 
		_ => item_rarity.Common 
		# ...

<# ... Add here the textures for the rarity thumbnail background ...#>
FindRarityThumbnailBackgroundTexture(Id:item_rarity) : texture =
	case(Id):
		item_rarity.Common => Textures.RewardThumbnail.T_IconBG_Common
		item_rarity.Uncommon => Textures.RewardThumbnail.T_IconBG_Uncommon
		item_rarity.Rare => Textures.RewardThumbnail.T_IconBG_Rare
		item_rarity.Epic => Textures.RewardThumbnail.T_IconBG_Epic
		item_rarity.Legendary => Textures.RewardThumbnail.T_IconBG_Legendary
		item_rarity.Mythic => Textures.RewardThumbnail.T_IconBG_Mythic
		item_rarity.Exotic => Textures.RewardThumbnail.T_IconBG_Exotic
		# ...

<# ... Add here the textures for the rarity card background ...#>
FindRarityCardBackgroundTexture(Id:item_rarity) : texture =
	case(Id):
		item_rarity.Common => Textures.RewardBackgrounds.T_CardBG_Common
		item_rarity.Uncommon => Textures.RewardBackgrounds.T_CardBG_Uncommon
		item_rarity.Rare => Textures.RewardBackgrounds.T_CardBG_Rare
		item_rarity.Epic => Textures.RewardBackgrounds.T_CardBG_Epic
		item_rarity.Legendary => Textures.RewardBackgrounds.T_CardBG_Legendary
		item_rarity.Mythic => Textures.RewardBackgrounds.T_CardBG_Mythic
		item_rarity.Exotic => Textures.RewardBackgrounds.T_CardBG_Exotic
		# ...

<# 
ATTRIBUTES
#>
item_attribute := enum:
	Quantity
	DPS
	Damage
	MagSize
	FireRate
	Reload
	Health
	Duration
	Shield

ToString(Id:item_attribute) : string =
	case(Id):
		item_attribute.Quantity => "Quantity"
		item_attribute.DPS => "DPS"
		item_attribute.Damage => "Dmg"
		item_attribute.MagSize => "Mag Size"
		item_attribute.FireRate => "Fire Rate"
		item_attribute.Reload => "Reload"
		item_attribute.Health => "HP"
		item_attribute.Duration => "Duration"
		item_attribute.Shield => "Shield"

item_attribute_data := struct<concrete>:
	Id:item_attribute = item_attribute.Quantity

	Value:float = 0.0

GetAttributeText(ItemAttributesData:[]item_attribute_data): string=
	var Attributes:string = ""
	for (CurrentAttribute : ItemAttributesData):
		CurrentText := "{ToString(CurrentAttribute.Id)} : {ToSmartString(CurrentAttribute.Value,2)}\n"
		set Attributes += CurrentText
	return Attributes
	
<# 
ITEM IDs
... Add here all the items you have in your project ...
L1 Common L2 Uncommon L3 Rare L4 Epic L5 Legendary
NOTE After this enum has been useed, new ones must be added in the last position so they are not overrided
#>

# Rifles
item_rifle_id := enum:
    Undefined

	BurstAssaultRifle_L1 #0
	BurstAssaultRifle_L2 #1
	BurstAssaultRifle_L3 #2
	BurstAssaultRifle_L4 #3
	BurstAssaultRifle_L5 #4
	
	AssaultRifle_L1 #5
	AssaultRifle_L2 #6
	AssaultRifle_L3 #7
	AssaultRifle_L4 #8
	AssaultRifle_L5 #9
	
	HeavyAssaultRifle_L1 #10
	HeavyAssaultRifle_L2 #11
	HeavyAssaultRifle_L3 #12
	HeavyAssaultRifle_L4 #13
	HeavyAssaultRifle_L5 #14
	
	InfantryRifle_L1 #15
	InfantryRifle_L2 #16
	InfantryRifle_L3 #17
	InfantryRifle_L4 #18
	InfantryRifle_L5 #19
	
	RangerAssaultRifle_L1 #20
	RangerAssaultRifle_L2 #21
	RangerAssaultRifle_L3 #22
	RangerAssaultRifle_L4 #23
	RangerAssaultRifle_L5 #24
	RangerAssaultRifle_L6 #25
	
	WarforgedAssaultRifle_L1 #26
	WarforgedAssaultRifle_L2 #27
	WarforgedAssaultRifle_L3 #28
	WarforgedAssaultRifle_L4 #29
	WarforgedAssaultRifle_L5 #30

# Shotguns
item_shotgun_id := enum:
	Undefined

	AutoShotgun_L1 #0
	AutoShotgun_L2 #1
	AutoShotgun_L3 #2
	AutoShotgun_L4 #3
	AutoShotgun_L5 #4
	AutoShotgun_L6 #5
	
	CombatShotgun_L1 #6
	CombatShotgun_L2 #7
	CombatShotgun_L3 #8
	CombatShotgun_L4 #9
	CombatShotgun_L5 #10
	
	DrumShotgun_L1 #11
	DrumShotgun_L2 #12
	DrumShotgun_L3 #13
	DrumShotgun_L4 #14
	DrumShotgun_L5 #15
	
	GatekeeperShotgun_L1 #16
	GatekeeperShotgun_L2 #17
	GatekeeperShotgun_L3 #18
	GatekeeperShotgun_L4 #19
	GatekeeperShotgun_L5 #20
	
	SovereignShotgun_L1 #21
	SovereignShotgun_L2 #22
	SovereignShotgun_L3 #23
	SovereignShotgun_L4 #24
	SovereignShotgun_L5 #25

	LeverActionShotgun_L1 #26
	LeverActionShotgun_L2 #27
	LeverActionShotgun_L3 #28
	LeverActionShotgun_L4 #29
	LeverActionShotgun_L5 #30
	
	PumpShotgun_L1 #31
	PumpShotgun_L2 #32
	PumpShotgun_L3 #33
	PumpShotgun_L4 #34
	PumpShotgun_L5 #35
	
	DoubleBarrelShotgun_L4 #36
	DoubleBarrelShotgun_L5 #37
	
	MakeshiftShotgun_L1 #38
	MakeshiftShotgun_L2 #39
	MakeshiftShotgun_L3 #40

# SMGs
item_smg_id := enum:
	Undefined

	CombatSMG_L1 #0
	CombatSMG_L2 #1
	CombatSMG_L3 #2
	CombatSMG_L4 #3
	CombatSMG_L5 #4
	CombatSMG_L6 #5
	
	DualMicroSMGs_L1 #6
	DualMicroSMGs_L2 #7
	DualMicroSMGs_L3 #8
	DualMicroSMGs_L4 #9
	DualMicroSMGs_L5 #10
	
	HarbingerSMG_L1 #11
	HarbingerSMG_L2 #12
	HarbingerSMG_L3 #13
	HarbingerSMG_L4 #14
	HarbingerSMG_L5 #15
	
	RapidFireSMG_L1 #16
	RapidFireSMG_L2 #17
	RapidFireSMG_L3 #18
	RapidFireSMG_L4 #19
	RapidFireSMG_L5 #20
	RapidFireSMG_L6 #21
	
	ThunderBurstSMG_L1 #22
	ThunderBurstSMG_L2 #23
	ThunderBurstSMG_L3 #24
	ThunderBurstSMG_L4 #25
	ThunderBurstSMG_L5 #26
	
	TwinMagSMG_L1 #27
	TwinMagSMG_L2 #28
	TwinMagSMG_L3 #29
	TwinMagSMG_L4 #30
	TwinMagSMG_L5 #31

# Pistols
item_pistol_id := enum:
	Undefined

	DualPistols_L2 #0
	DualPistols_L3 #1
	DualPistols_L4 #2
	DualPistols_L5 #3
	
	HandCannon_L3 #4
	HandCannon_L4 #5
	HandCannon_L5 #6
	
	MonarchPistol_L1 #7
	MonarchPistol_L2 #8
	MonarchPistol_L3 #9
	MonarchPistol_L4 #10
	MonarchPistol_L5 #11
	
	RangerPistol_L1 #12
	RangerPistol_L2 #13
	RangerPistol_L3 #14
	RangerPistol_L4 #15
	RangerPistol_L5 #16
	
	Revolver_L1 #17
	Revolver_L2 #18
	Revolver_L3 #19
	Revolver_L4 #20
	Revolver_L5 #21
	
	SixShooter_L2 #22
	SixShooter_L3 #23
	SixShooter_L4 #24
	
	TacticalPistol_L1 #25
	TacticalPistol_L2 #26
	TacticalPistol_L3 #27
	TacticalPistol_L4 #28
	TacticalPistol_L5 #29
	TacticalPistol_L6 #30

# DMRs
item_dmr_id := enum:
	Undefined

	CobraDMR_L1 #0
	CobraDMR_L2 #1
	CobraDMR_L3 #2
	CobraDMR_L4 #3
	CobraDMR_L5 #4
	CobraDMR_L6 #5

	ThermalDMR_L1 #6
	ThermalDMR_L2 #7
	ThermalDMR_L3 #8
	ThermalDMR_L4 #9
	ThermalDMR_L5 #10
	ThermalDMR_L6 #11

# Snipers
item_sniper_id := enum:
	Undefined

	AutomaticSniperRifle_L2 #0
	AutomaticSniperRifle_L3 #1
	AutomaticSniperRifle_L4 #2
	AutomaticSniperRifle_L5 #3
	
	BoltActionSniperRifle_L1 #4
	BoltActionSniperRifle_L2 #5
	BoltActionSniperRifle_L3 #6
	BoltActionSniperRifle_L4 #7
	BoltActionSniperRifle_L5 #8
	
	CobraDMR_L1 #9
	CobraDMR_L2 #10
	CobraDMR_L3 #11
	CobraDMR_L4 #12
	CobraDMR_L5 #13
	CobraDMR_L6 #14
	
	ExCaliberRifle_L1 #15
	ExCaliberRifle_L2 #16
	ExCaliberRifle_L3 #17
	ExCaliberRifle_L4 #18
	ExCaliberRifle_L5 #19
	
	HuntressDMR_L2 #20
	HuntressDMR_L3 #21
	HuntressDMR_L4 #22
	HuntressDMR_L5 #23
	
	ReaperSniperRifle_L2 #24
	ReaperSniperRifle_L3 #25
	ReaperSniperRifle_L4 #26
	ReaperSniperRifle_L5 #27

# Explosives
item_explosive_id := enum:
	Undefined

	BoomBolt_L3 #0
	BoomBolt_L4 #1
	BoomBolt_L5 #2
	BoomBolt_L6 #3
	BoomBow_L5 #4
	Dynamite_L2 #5
	ExplosiveRepeaterRifle_L3 #6
	ExplosiveRepeaterRifle_L4 #7
	ExplosiveRepeaterRifle_L5 #8
	FireflyJar_L3 #9
	Grenade_L1 #10
	MechanicalExplosiveBow_L3 #11
	MechanicalExplosiveBow_L4 #12
	MechanicalExplosiveBow_L5 #13
	ProximityMine_L2 #14
	ProximityMine_L4 #15
	RemoteExplosive_L4 #16
	StickyGrenadeLauncher_L4 #17
	StickyGrenadeLauncher_L5 #18

# Others
item_others_id := enum:
	Undefined

	# Bows
	BasicHammer_L1 #0
	BasicSword_L1 #1
	Chainsaw_L4 #2
	Crossbow_L3 #3
	Crossbow_L4 #4
	KineticBlade_L3 #5
	KineticBlade_L4 #6
	MakeshiftBow_L2 #7
	MechanicalBow_L3 #8

# Consumables
item_consumable_id := enum:
	Undefined

	# Healing
	Bandage_L1 #0
	ChugJug_L5 #1
	Flopper_L2 #2
	Medkit_L2 #3
	Mushroom_L1 #4
	ShieldPotion_L3 #5
	Shieldfish_L3 #6
	SlurpJuice_L4 #7
	SlurpMushroom_L1 #8
	Slurpfish_L4 #9
	SmallFry_L1 #10
	SmallShieldPotion_L2 #11
	SpicyFish_L3 #12

	# Offensive
	Grenade_L1 # 5
	Dynamite_L2 # 6
	ProximityMine_L3 # 7
	Recycler_L4 # 8
	Recycler_L5 # 9

<# 
TEXTURES
... Add here the textures for each category ...
#>
# Rifles
FindRifleIconTexture(Id:item_rifle_id)<computes> : texture =
	case(Id):
		# Rifle 1
		item_rifle_id.BurstAssaultRifle_L1=> Textures.ItemIcons.RifleIcons.T_BurstAsRifle
		item_rifle_id.BurstAssaultRifle_L2=> Textures.ItemIcons.RifleIcons.T_BurstAsRifle
		item_rifle_id.BurstAssaultRifle_L3=> Textures.ItemIcons.RifleIcons.T_BurstAsRifle
		item_rifle_id.BurstAssaultRifle_L4=> Textures.ItemIcons.RifleIcons.T_BurstAsRifle
		item_rifle_id.BurstAssaultRifle_L5=> Textures.ItemIcons.RifleIcons.T_BurstAsRifle
		
		item_rifle_id.AssaultRifle_L1=> Textures.ItemIcons.RifleIcons.T_AssaultRifleB
		item_rifle_id.AssaultRifle_L2=> Textures.ItemIcons.RifleIcons.T_AssaultRifleB
		item_rifle_id.AssaultRifle_L3=> Textures.ItemIcons.RifleIcons.T_AssaultRifleB
		item_rifle_id.AssaultRifle_L4=> Textures.ItemIcons.RifleIcons.T_AssaultRifleB
		item_rifle_id.AssaultRifle_L5=> Textures.ItemIcons.RifleIcons.T_AssaultRifleB
		
		item_rifle_id.HeavyAssaultRifle_L1=> Textures.ItemIcons.RifleIcons.T_HeavyAssaultRifle
		item_rifle_id.HeavyAssaultRifle_L2=> Textures.ItemIcons.RifleIcons.T_HeavyAssaultRifle
		item_rifle_id.HeavyAssaultRifle_L3=> Textures.ItemIcons.RifleIcons.T_HeavyAssaultRifle
		item_rifle_id.HeavyAssaultRifle_L4=> Textures.ItemIcons.RifleIcons.T_HeavyAssaultRifle
		item_rifle_id.HeavyAssaultRifle_L5=> Textures.ItemIcons.RifleIcons.T_HeavyAssaultRifle
		
		item_rifle_id.InfantryRifle_L1=> Textures.ItemIcons.RifleIcons.T_InfantryRifle
		item_rifle_id.InfantryRifle_L2=> Textures.ItemIcons.RifleIcons.T_InfantryRifle
		item_rifle_id.InfantryRifle_L3=> Textures.ItemIcons.RifleIcons.T_InfantryRifle
		item_rifle_id.InfantryRifle_L4=> Textures.ItemIcons.RifleIcons.T_InfantryRifle
		item_rifle_id.InfantryRifle_L5=> Textures.ItemIcons.RifleIcons.T_InfantryRifle
		
		item_rifle_id.RangerAssaultRifle_L1=> Textures.ItemIcons.RifleIcons.T_RangerAssaultRifle
		item_rifle_id.RangerAssaultRifle_L2=> Textures.ItemIcons.RifleIcons.T_RangerAssaultRifle
		item_rifle_id.RangerAssaultRifle_L3=> Textures.ItemIcons.RifleIcons.T_RangerAssaultRifle
		item_rifle_id.RangerAssaultRifle_L4=> Textures.ItemIcons.RifleIcons.T_RangerAssaultRifle
		item_rifle_id.RangerAssaultRifle_L5=> Textures.ItemIcons.RifleIcons.T_RangerAssaultRifle
		item_rifle_id.RangerAssaultRifle_L6=> Textures.ItemIcons.RifleIcons.T_RangerAssaultRifle
		
		item_rifle_id.WarforgedAssaultRifle_L1=> Textures.ItemIcons.RifleIcons.T_WarforgedAssaultRifle
		item_rifle_id.WarforgedAssaultRifle_L2=> Textures.ItemIcons.RifleIcons.T_WarforgedAssaultRifle
		item_rifle_id.WarforgedAssaultRifle_L3=> Textures.ItemIcons.RifleIcons.T_WarforgedAssaultRifle
		item_rifle_id.WarforgedAssaultRifle_L4=> Textures.ItemIcons.RifleIcons.T_WarforgedAssaultRifle
		item_rifle_id.WarforgedAssaultRifle_L5=> Textures.ItemIcons.RifleIcons.T_WarforgedAssaultRifle

		# Rifle 3
		# ...

		_ => Textures.RewardBackgrounds.T_CardBG_Common

# Shotguns
FindShotgunIconTexture(Id:item_shotgun_id)<computes> : texture =
	case(Id):
		# Shotgun 1
		item_shotgun_id.AutoShotgun_L1=> Textures.ItemIcons.ShotgunIcons.T_AutoShotgun
		item_shotgun_id.AutoShotgun_L2=> Textures.ItemIcons.ShotgunIcons.T_AutoShotgun
		item_shotgun_id.AutoShotgun_L3=> Textures.ItemIcons.ShotgunIcons.T_AutoShotgun
		item_shotgun_id.AutoShotgun_L4=> Textures.ItemIcons.ShotgunIcons.T_AutoShotgun
		item_shotgun_id.AutoShotgun_L5=> Textures.ItemIcons.ShotgunIcons.T_AutoShotgun
		item_shotgun_id.AutoShotgun_L6=> Textures.ItemIcons.ShotgunIcons.T_AutoShotgun
		
		item_shotgun_id.CombatShotgun_L1=> Textures.ItemIcons.ShotgunIcons.T_CombatShotgun
		item_shotgun_id.CombatShotgun_L2=> Textures.ItemIcons.ShotgunIcons.T_CombatShotgun
		item_shotgun_id.CombatShotgun_L3=> Textures.ItemIcons.ShotgunIcons.T_CombatShotgun
		item_shotgun_id.CombatShotgun_L4=> Textures.ItemIcons.ShotgunIcons.T_CombatShotgun
		item_shotgun_id.CombatShotgun_L5=> Textures.ItemIcons.ShotgunIcons.T_CombatShotgun
		
		item_shotgun_id.DrumShotgun_L1=> Textures.ItemIcons.ShotgunIcons.T_DrumShotgun
		item_shotgun_id.DrumShotgun_L2=> Textures.ItemIcons.ShotgunIcons.T_DrumShotgun
		item_shotgun_id.DrumShotgun_L3=> Textures.ItemIcons.ShotgunIcons.T_DrumShotgun
		item_shotgun_id.DrumShotgun_L4=> Textures.ItemIcons.ShotgunIcons.T_DrumShotgun
		item_shotgun_id.DrumShotgun_L5=> Textures.ItemIcons.ShotgunIcons.T_DrumShotgun
		
		item_shotgun_id.GatekeeperShotgun_L1=> Textures.ItemIcons.ShotgunIcons.T_GatekeeperShotgun
		item_shotgun_id.GatekeeperShotgun_L2=> Textures.ItemIcons.ShotgunIcons.T_GatekeeperShotgun
		item_shotgun_id.GatekeeperShotgun_L3=> Textures.ItemIcons.ShotgunIcons.T_GatekeeperShotgun
		item_shotgun_id.GatekeeperShotgun_L4=> Textures.ItemIcons.ShotgunIcons.T_GatekeeperShotgun
		item_shotgun_id.GatekeeperShotgun_L5=> Textures.ItemIcons.ShotgunIcons.T_GatekeeperShotgun
		
		item_shotgun_id.SovereignShotgun_L1=> Textures.ItemIcons.ShotgunIcons.T_SovereignShotgun
		item_shotgun_id.SovereignShotgun_L2=> Textures.ItemIcons.ShotgunIcons.T_SovereignShotgun
		item_shotgun_id.SovereignShotgun_L3=> Textures.ItemIcons.ShotgunIcons.T_SovereignShotgun
		item_shotgun_id.SovereignShotgun_L4=> Textures.ItemIcons.ShotgunIcons.T_SovereignShotgun
		item_shotgun_id.SovereignShotgun_L5=> Textures.ItemIcons.ShotgunIcons.T_SovereignShotgun

		item_shotgun_id.LeverActionShotgun_L1=> Textures.ItemIcons.ShotgunIcons.T_LeverActionShotgun
		item_shotgun_id.LeverActionShotgun_L2=> Textures.ItemIcons.ShotgunIcons.T_LeverActionShotgun
		item_shotgun_id.LeverActionShotgun_L3=> Textures.ItemIcons.ShotgunIcons.T_LeverActionShotgun
		item_shotgun_id.LeverActionShotgun_L4=> Textures.ItemIcons.ShotgunIcons.T_LeverActionShotgun
		item_shotgun_id.LeverActionShotgun_L5=> Textures.ItemIcons.ShotgunIcons.T_LeverActionShotgun
		
		item_shotgun_id.PumpShotgun_L1=> Textures.ItemIcons.ShotgunIcons.T_PumpShotgun
		item_shotgun_id.PumpShotgun_L2=> Textures.ItemIcons.ShotgunIcons.T_PumpShotgun
		item_shotgun_id.PumpShotgun_L3=> Textures.ItemIcons.ShotgunIcons.T_PumpShotgun
		item_shotgun_id.PumpShotgun_L4=> Textures.ItemIcons.ShotgunIcons.T_PumpShotgun
		item_shotgun_id.PumpShotgun_L5=> Textures.ItemIcons.ShotgunIcons.T_PumpShotgun
		
		item_shotgun_id.DoubleBarrelShotgun_L4=> Textures.ItemIcons.ShotgunIcons.T_DoubleBarrelShotgun
		item_shotgun_id.DoubleBarrelShotgun_L5=> Textures.ItemIcons.ShotgunIcons.T_DoubleBarrelShotgun
		
		item_shotgun_id.MakeshiftShotgun_L1=> Textures.ItemIcons.ShotgunIcons.T_MakeshiftShotgun
		item_shotgun_id.MakeshiftShotgun_L2=> Textures.ItemIcons.ShotgunIcons.T_MakeshiftShotgun
		item_shotgun_id.MakeshiftShotgun_L3=> Textures.ItemIcons.ShotgunIcons.T_MakeshiftShotgun

		# Shotgun 3
		#...

		_ => Textures.RewardBackgrounds.T_CardBG_Common

# SMGs
FindSMGIconTexture(Id:item_smg_id)<computes> : texture =
	case(Id):
		# SMG 1
		item_smg_id.CombatSMG_L1=> Textures.ItemIcons.SMGIcons.T_CombatSMG
		item_smg_id.CombatSMG_L2=> Textures.ItemIcons.SMGIcons.T_CombatSMG
		item_smg_id.CombatSMG_L3=> Textures.ItemIcons.SMGIcons.T_CombatSMG
		item_smg_id.CombatSMG_L4=> Textures.ItemIcons.SMGIcons.T_CombatSMG
		item_smg_id.CombatSMG_L5=> Textures.ItemIcons.SMGIcons.T_CombatSMG
		item_smg_id.CombatSMG_L6=> Textures.ItemIcons.SMGIcons.T_CombatSMG
		
		item_smg_id.DualMicroSMGs_L1=> Textures.ItemIcons.SMGIcons.T_DualMicroSMGs
		item_smg_id.DualMicroSMGs_L2=> Textures.ItemIcons.SMGIcons.T_DualMicroSMGs
		item_smg_id.DualMicroSMGs_L3=> Textures.ItemIcons.SMGIcons.T_DualMicroSMGs
		item_smg_id.DualMicroSMGs_L4=> Textures.ItemIcons.SMGIcons.T_DualMicroSMGs
		item_smg_id.DualMicroSMGs_L5=> Textures.ItemIcons.SMGIcons.T_DualMicroSMGs
		
		item_smg_id.HarbingerSMG_L1=> Textures.ItemIcons.SMGIcons.T_HarbingerSMG
		item_smg_id.HarbingerSMG_L2=> Textures.ItemIcons.SMGIcons.T_HarbingerSMG
		item_smg_id.HarbingerSMG_L3=> Textures.ItemIcons.SMGIcons.T_HarbingerSMG
		item_smg_id.HarbingerSMG_L4=> Textures.ItemIcons.SMGIcons.T_HarbingerSMG
		item_smg_id.HarbingerSMG_L5=> Textures.ItemIcons.SMGIcons.T_HarbingerSMG
		
		item_smg_id.RapidFireSMG_L1=> Textures.ItemIcons.SMGIcons.T_RapidfireSMG
		item_smg_id.RapidFireSMG_L2=> Textures.ItemIcons.SMGIcons.T_RapidfireSMG
		item_smg_id.RapidFireSMG_L3=> Textures.ItemIcons.SMGIcons.T_RapidfireSMG
		item_smg_id.RapidFireSMG_L4=> Textures.ItemIcons.SMGIcons.T_RapidfireSMG
		item_smg_id.RapidFireSMG_L5=> Textures.ItemIcons.SMGIcons.T_RapidfireSMG
		item_smg_id.RapidFireSMG_L6=> Textures.ItemIcons.SMGIcons.T_RapidfireSMG
		
		item_smg_id.ThunderBurstSMG_L1=> Textures.ItemIcons.SMGIcons.T_ThunderBurstSMG
		item_smg_id.ThunderBurstSMG_L2=> Textures.ItemIcons.SMGIcons.T_ThunderBurstSMG
		item_smg_id.ThunderBurstSMG_L3=> Textures.ItemIcons.SMGIcons.T_ThunderBurstSMG
		item_smg_id.ThunderBurstSMG_L4=> Textures.ItemIcons.SMGIcons.T_ThunderBurstSMG
		item_smg_id.ThunderBurstSMG_L5=> Textures.ItemIcons.SMGIcons.T_ThunderBurstSMG
		
		item_smg_id.TwinMagSMG_L1=> Textures.ItemIcons.SMGIcons.T_TwinMagSMG
		item_smg_id.TwinMagSMG_L2=> Textures.ItemIcons.SMGIcons.T_TwinMagSMG
		item_smg_id.TwinMagSMG_L3=> Textures.ItemIcons.SMGIcons.T_TwinMagSMG
		item_smg_id.TwinMagSMG_L4=> Textures.ItemIcons.SMGIcons.T_TwinMagSMG
		item_smg_id.TwinMagSMG_L5=> Textures.ItemIcons.SMGIcons.T_TwinMagSMG
		
		# SMG 3
		# ...

		_ => Textures.RewardBackgrounds.T_CardBG_Common

# Pistols
FindPistolIconTexture(Id:item_pistol_id)<computes> : texture =
	case(Id):
		# Pistol 1
		
		item_pistol_id.DualPistols_L2=> Textures.ItemIcons.PistolIcons.T_DualPistols
		item_pistol_id.DualPistols_L3=> Textures.ItemIcons.PistolIcons.T_DualPistols
		item_pistol_id.DualPistols_L4=> Textures.ItemIcons.PistolIcons.T_DualPistols
		item_pistol_id.DualPistols_L5=> Textures.ItemIcons.PistolIcons.T_DualPistols
		
		item_pistol_id.HandCannon_L3=> Textures.ItemIcons.PistolIcons.T_HandCannon
		item_pistol_id.HandCannon_L4=> Textures.ItemIcons.PistolIcons.T_HandCannon
		item_pistol_id.HandCannon_L5=> Textures.ItemIcons.PistolIcons.T_HandCannon
		
		item_pistol_id.MonarchPistol_L1=> Textures.ItemIcons.PistolIcons.T_MonarchPistol
		item_pistol_id.MonarchPistol_L2=> Textures.ItemIcons.PistolIcons.T_MonarchPistol
		item_pistol_id.MonarchPistol_L3=> Textures.ItemIcons.PistolIcons.T_MonarchPistol
		item_pistol_id.MonarchPistol_L4=> Textures.ItemIcons.PistolIcons.T_MonarchPistol
		item_pistol_id.MonarchPistol_L5=> Textures.ItemIcons.PistolIcons.T_MonarchPistol
		
		item_pistol_id.RangerPistol_L1=> Textures.ItemIcons.PistolIcons.T_RangerPistol
		item_pistol_id.RangerPistol_L2=> Textures.ItemIcons.PistolIcons.T_RangerPistol
		item_pistol_id.RangerPistol_L3=> Textures.ItemIcons.PistolIcons.T_RangerPistol
		item_pistol_id.RangerPistol_L4=> Textures.ItemIcons.PistolIcons.T_RangerPistol
		item_pistol_id.RangerPistol_L5=> Textures.ItemIcons.PistolIcons.T_RangerPistol
		
		item_pistol_id.Revolver_L1=> Textures.ItemIcons.PistolIcons.T_Revolver
		item_pistol_id.Revolver_L2=> Textures.ItemIcons.PistolIcons.T_Revolver
		item_pistol_id.Revolver_L3=> Textures.ItemIcons.PistolIcons.T_Revolver
		item_pistol_id.Revolver_L4=> Textures.ItemIcons.PistolIcons.T_Revolver
		item_pistol_id.Revolver_L5=> Textures.ItemIcons.PistolIcons.T_Revolver
		
		item_pistol_id.SixShooter_L2=> Textures.ItemIcons.PistolIcons.T_SixShooter
		item_pistol_id.SixShooter_L3=> Textures.ItemIcons.PistolIcons.T_SixShooter
		item_pistol_id.SixShooter_L4=> Textures.ItemIcons.PistolIcons.T_SixShooter
		
		item_pistol_id.TacticalPistol_L1=> Textures.ItemIcons.PistolIcons.T_TacticalPistol
		item_pistol_id.TacticalPistol_L2=> Textures.ItemIcons.PistolIcons.T_TacticalPistol
		item_pistol_id.TacticalPistol_L3=> Textures.ItemIcons.PistolIcons.T_TacticalPistol
		item_pistol_id.TacticalPistol_L4=> Textures.ItemIcons.PistolIcons.T_TacticalPistol
		item_pistol_id.TacticalPistol_L5=> Textures.ItemIcons.PistolIcons.T_TacticalPistol
		item_pistol_id.TacticalPistol_L6=> Textures.ItemIcons.PistolIcons.T_TacticalPistol
		# Pistol 3
		# ...

		_ => Textures.RewardBackgrounds.T_CardBG_Common

# DMRs
FindDMRIconTexture(Id:item_dmr_id)<computes> : texture =
	case(Id):
		# DMR 1
		item_dmr_id.CobraDMR_L1=> Textures.ItemIcons.DMRIcons.T_CobraDMR
		item_dmr_id.CobraDMR_L2=> Textures.ItemIcons.DMRIcons.T_CobraDMR
		item_dmr_id.CobraDMR_L3=> Textures.ItemIcons.DMRIcons.T_CobraDMR
		item_dmr_id.CobraDMR_L4=> Textures.ItemIcons.DMRIcons.T_CobraDMR
		item_dmr_id.CobraDMR_L5=> Textures.ItemIcons.DMRIcons.T_CobraDMR
		item_dmr_id.CobraDMR_L6=> Textures.ItemIcons.DMRIcons.T_CobraDMR

		# DMR 2
		item_dmr_id.ThermalDMR_L1=> Textures.ItemIcons.DMRIcons.T_ThermalDMR
		item_dmr_id.ThermalDMR_L2=> Textures.ItemIcons.DMRIcons.T_ThermalDMR
		item_dmr_id.ThermalDMR_L3=> Textures.ItemIcons.DMRIcons.T_ThermalDMR
		item_dmr_id.ThermalDMR_L4=> Textures.ItemIcons.DMRIcons.T_ThermalDMR
		item_dmr_id.ThermalDMR_L5=> Textures.ItemIcons.DMRIcons.T_ThermalDMR
		item_dmr_id.ThermalDMR_L6=> Textures.ItemIcons.DMRIcons.T_ThermalDMR

		# DMR 3
		# ..

		_ => Textures.RewardBackgrounds.T_CardBG_Common

# Snipers
FindSniperIconTexture(Id:item_sniper_id)<computes> : texture =
	case(Id):
		# Sniper 1
		item_sniper_id.AutomaticSniperRifle_L2=> Textures.ItemIcons.SniperIcons.T_AutomaticSniperRifle
		item_sniper_id.AutomaticSniperRifle_L3=> Textures.ItemIcons.SniperIcons.T_AutomaticSniperRifle
		item_sniper_id.AutomaticSniperRifle_L4=> Textures.ItemIcons.SniperIcons.T_AutomaticSniperRifle
		item_sniper_id.AutomaticSniperRifle_L5=> Textures.ItemIcons.SniperIcons.T_AutomaticSniperRifle
		item_sniper_id.BoltActionSniperRifle_L1=> Textures.ItemIcons.SniperIcons.T_BoltActionSniperRifle
		item_sniper_id.BoltActionSniperRifle_L2=> Textures.ItemIcons.SniperIcons.T_BoltActionSniperRifle
		item_sniper_id.BoltActionSniperRifle_L3=> Textures.ItemIcons.SniperIcons.T_BoltActionSniperRifle
		item_sniper_id.BoltActionSniperRifle_L4=> Textures.ItemIcons.SniperIcons.T_BoltActionSniperRifle
		item_sniper_id.BoltActionSniperRifle_L5=> Textures.ItemIcons.SniperIcons.T_BoltActionSniperRifle
		item_sniper_id.CobraDMR_L1=> Textures.ItemIcons.SniperIcons.T_CobraDMR
		item_sniper_id.CobraDMR_L2=> Textures.ItemIcons.SniperIcons.T_CobraDMR
		item_sniper_id.CobraDMR_L3=> Textures.ItemIcons.SniperIcons.T_CobraDMR
		item_sniper_id.CobraDMR_L4=> Textures.ItemIcons.SniperIcons.T_CobraDMR
		item_sniper_id.CobraDMR_L5=> Textures.ItemIcons.SniperIcons.T_CobraDMR
		item_sniper_id.CobraDMR_L6=> Textures.ItemIcons.SniperIcons.T_CobraDMR
		item_sniper_id.ExCaliberRifle_L1=> Textures.ItemIcons.SniperIcons.T_ExCaliberRifle
		item_sniper_id.ExCaliberRifle_L2=> Textures.ItemIcons.SniperIcons.T_ExCaliberRifle
		item_sniper_id.ExCaliberRifle_L3=> Textures.ItemIcons.SniperIcons.T_ExCaliberRifle
		item_sniper_id.ExCaliberRifle_L4=> Textures.ItemIcons.SniperIcons.T_ExCaliberRifle
		item_sniper_id.ExCaliberRifle_L5=> Textures.ItemIcons.SniperIcons.T_ExCaliberRifle
		item_sniper_id.HuntressDMR_L2=> Textures.ItemIcons.SniperIcons.T_HuntressDMR
		item_sniper_id.HuntressDMR_L3=> Textures.ItemIcons.SniperIcons.T_HuntressDMR
		item_sniper_id.HuntressDMR_L4=> Textures.ItemIcons.SniperIcons.T_HuntressDMR
		item_sniper_id.HuntressDMR_L5=> Textures.ItemIcons.SniperIcons.T_HuntressDMR
		item_sniper_id.ReaperSniperRifle_L2=> Textures.ItemIcons.SniperIcons.T_ReaperSniperRifle
		item_sniper_id.ReaperSniperRifle_L3=> Textures.ItemIcons.SniperIcons.T_ReaperSniperRifle
		item_sniper_id.ReaperSniperRifle_L4=> Textures.ItemIcons.SniperIcons.T_ReaperSniperRifle
		item_sniper_id.ReaperSniperRifle_L5=> Textures.ItemIcons.SniperIcons.T_ReaperSniperRifle

		# Sniper 3
		# ..

		_ => Textures.RewardBackgrounds.T_CardBG_Common

# Explosives
FindExplosiveIconTexture(Id:item_explosive_id)<computes> : texture =
	case(Id):
		# Explosive 1
		item_explosive_id.BoomBolt_L3=> Textures.ItemIcons.ExplosiveIcons.T_BoomBolt
		item_explosive_id.BoomBolt_L4=> Textures.ItemIcons.ExplosiveIcons.T_BoomBolt
		item_explosive_id.BoomBolt_L5=> Textures.ItemIcons.ExplosiveIcons.T_BoomBolt
		item_explosive_id.BoomBolt_L6=> Textures.ItemIcons.ExplosiveIcons.T_BoomBolt
		item_explosive_id.BoomBow_L5=> Textures.ItemIcons.ExplosiveIcons.T_BoomBow
		item_explosive_id.Dynamite_L2=> Textures.ItemIcons.ExplosiveIcons.T_Dynamite
		item_explosive_id.ExplosiveRepeaterRifle_L3=> Textures.ItemIcons.ExplosiveIcons.T_ExplosiveRepeaterRifle
		item_explosive_id.ExplosiveRepeaterRifle_L4=> Textures.ItemIcons.ExplosiveIcons.T_ExplosiveRepeaterRifle
		item_explosive_id.ExplosiveRepeaterRifle_L5=> Textures.ItemIcons.ExplosiveIcons.T_ExplosiveRepeaterRifle
		item_explosive_id.FireflyJar_L3=> Textures.ItemIcons.ExplosiveIcons.T_FireflyJar
		item_explosive_id.Grenade_L1=> Textures.ItemIcons.ExplosiveIcons.T_Grenade
		item_explosive_id.MechanicalExplosiveBow_L3=> Textures.ItemIcons.ExplosiveIcons.T_MechanicalExplosiveBow
		item_explosive_id.MechanicalExplosiveBow_L4=> Textures.ItemIcons.ExplosiveIcons.T_MechanicalExplosiveBow
		item_explosive_id.MechanicalExplosiveBow_L5=> Textures.ItemIcons.ExplosiveIcons.T_MechanicalExplosiveBow
		item_explosive_id.ProximityMine_L2=> Textures.ItemIcons.ExplosiveIcons.T_ProximityMine
		item_explosive_id.ProximityMine_L4=> Textures.ItemIcons.ExplosiveIcons.T_ProximityMine
		item_explosive_id.RemoteExplosive_L4=> Textures.ItemIcons.ExplosiveIcons.T_RemoteExplosive
		item_explosive_id.StickyGrenadeLauncher_L4=> Textures.ItemIcons.ExplosiveIcons.T_StickyGrenadeLauncher
		item_explosive_id.StickyGrenadeLauncher_L5=> Textures.ItemIcons.ExplosiveIcons.T_StickyGrenadeLauncher
		# Explosive 3
		# ..
		
		_ => Textures.RewardBackgrounds.T_CardBG_Common

# Others
FindOthersIconTexture(Id:item_others_id)<computes> : texture =
	case(Id):
		# 
		item_others_id.BasicHammer_L1=> Textures.ItemIcons.OthersIcons.T_Hammer
		item_others_id.BasicSword_L1=> Textures.ItemIcons.OthersIcons.T_BasicSword
		item_others_id.Chainsaw_L4=> Textures.ItemIcons.OthersIcons.T_Chainsaw
		
		item_others_id.Crossbow_L3=> Textures.ItemIcons.OthersIcons.T_Crossbow
		item_others_id.Crossbow_L4=> Textures.ItemIcons.OthersIcons.T_Crossbow
		
		item_others_id.KineticBlade_L3=> Textures.ItemIcons.OthersIcons.T_KineticBlade
		item_others_id.KineticBlade_L4=> Textures.ItemIcons.OthersIcons.T_KineticBlade
		
		item_others_id.MakeshiftBow_L2=> Textures.ItemIcons.OthersIcons.T_MakeshiftBow
		item_others_id.MechanicalBow_L3=> Textures.ItemIcons.OthersIcons.T_MechanicalBow
		
		# Melee 4
		# ..

		_ => Textures.RewardBackgrounds.T_CardBG_Common

# Consumables
FindConsumableIconTexture(Id:item_consumable_id)<computes> : texture =
	case(Id):
		# Healing
		item_consumable_id.Bandage_L1=> Textures.ItemIcons.ConsumableIcons.T_Bandage
		item_consumable_id.ChugJug_L5=> Textures.ItemIcons.ConsumableIcons.T_ChugJug
		item_consumable_id.Flopper_L2=> Textures.ItemIcons.ConsumableIcons.T_Flopper
		item_consumable_id.Medkit_L2=> Textures.ItemIcons.ConsumableIcons.T_Medkit
		item_consumable_id.Mushroom_L1=> Textures.ItemIcons.ConsumableIcons.T_Mushroom
		item_consumable_id.ShieldPotion_L3=> Textures.ItemIcons.ConsumableIcons.T_ShieldPotion
		item_consumable_id.Shieldfish_L3=> Textures.ItemIcons.ConsumableIcons.T_Shieldfish
		item_consumable_id.SlurpJuice_L4=> Textures.ItemIcons.ConsumableIcons.T_SlurpJuice
		item_consumable_id.SlurpMushroom_L1=> Textures.ItemIcons.ConsumableIcons.T_SlurpMushroom
		item_consumable_id.Slurpfish_L4=> Textures.ItemIcons.ConsumableIcons.T_Slurpfish
		item_consumable_id.SmallFry_L1=> Textures.ItemIcons.ConsumableIcons.T_SmallFry
		item_consumable_id.SmallShieldPotion_L2=> Textures.ItemIcons.ConsumableIcons.T_SmallShieldPotion
		item_consumable_id.SpicyFish_L3=> Textures.ItemIcons.ConsumableIcons.T_SpicyFish

		# Offensives
		item_consumable_id.Grenade_L1=> Textures.ItemIcons.ConsumableIcons.Offensives.T_Grenade
		item_consumable_id.Dynamite_L2=> Textures.ItemIcons.ConsumableIcons.Offensives.T_Dynamite	
		item_consumable_id.ProximityMine_L3=> Textures.ItemIcons.ConsumableIcons.Offensives.T_ProximityMine
		item_consumable_id.Recycler_L4=> Textures.ItemIcons.ConsumableIcons.Offensives.T_Recycler
		item_consumable_id.Recycler_L5=> Textures.ItemIcons.ConsumableIcons.Offensives.T_Recycler

		_ => Textures.RewardBackgrounds.T_CardBG_Common


item_data := class<concrete>(rewardable):
	@editable
	DisplayName:string = ""

	@editable
	Rarity:item_rarity = item_rarity.Common

	@editable # If Item granter defined then the index for this item there
	ItemGranterIndex:int = 0

	@editable
    Team:team_id = team_id.Blue

	@editable
	Attributes:[]item_attribute_data = array{}

	GetRarity<override>()<computes>: item_rarity = Rarity
	GetDisplayName<override>()<computes>: string = DisplayName