using { /Verse.org/Simulation }

<# 
Notations in numbers
#>

notation_definition := struct:
    Power:int
    String:string

# Data that has (Power, String) Value for the number notations
# NOTE Needs to be in ascendent order
NotationDefinitions:[]tuple(int, string) = array:
    (3      ,    "K")
    (6      ,    "M")
    (9      ,    "B")
    (12     ,    "T")
    (15     ,    "P")
    (18     ,    "E")

    (21     ,    "EK")
    (24     ,    "EM")
    (27     ,    "EB")
    (30     ,    "ET")
    (33     ,    "EP")
    (36     ,    "EE")

    (39     ,    "EEK")
    (42     ,    "EEM")
    (45     ,    "EEB")
    (48     ,    "EET")
    (51     ,    "EEP")
    (54     ,    "X") # EEE

    (57     ,    "XK")
    (60     ,    "XM")
    (63     ,    "XB")
    (66     ,    "XT")
    (69     ,    "XP")
    (72     ,    "XE")

    (75     ,    "XEK")
    (78     ,    "XEM")
    (81     ,    "XEB")
    (84     ,    "XET")
    (87     ,    "XEP")
    (90     ,    "XEE")

    (93     ,    "XEEK")
    (96     ,    "XEEM")
    (99     ,    "XEEB")
    (102    ,    "XEET")
    (105    ,    "XEEP")
    (108    ,    "XX") # XEEE

    (111    ,    "XXK")
    (114    ,    "XXM")
    (117    ,    "XXB")
    (120    ,    "XXT")
    (123    ,    "XXP")
    (126    ,    "XXE")

    (129    ,    "XXEK")
    (132    ,    "XXEM")
    (135    ,    "XXEB")
    (138    ,    "XXET")
    (141    ,    "XXEP")
    (144    ,    "XXEE")

    (147    ,    "XXEEK")
    (150    ,    "XXEEM")
    (153    ,    "XXEEB")
    (156    ,    "XXEET")
    (159    ,    "XXEEP")
    (162    ,    "Y") # XXEEE

    (165    ,    "YK")
    (168    ,    "YM")
    (171    ,    "YB")
    (174    ,    "YT")
    (177    ,    "YP")
    (180    ,    "YE")

    (183    ,    "YEK")
    (186    ,    "YEM")
    (189    ,    "YEB")
    (192    ,    "YET")
    (195    ,    "YEP")
    (198    ,    "YEE")

    (201    ,    "YEEK")
    (204    ,    "YEEM")
    (207    ,    "YEEB")
    (210    ,    "YEET")
    (213    ,    "YEEP")
    (216    ,    "YX") # YEEE

    (219    ,    "YXK")
    (222    ,    "YXM")
    (225    ,    "YXB")
    (228    ,    "YXT")
    (231    ,    "YXP")
    (234    ,    "YXE")

    (237    ,    "YXEK")
    (240    ,    "YXEM")
    (243    ,    "YXEB")
    (246    ,    "YXET")
    (249    ,    "YXEP")
    (252    ,    "YXEE")

    (255    ,    "YXEEK")
    (258    ,    "YXEEM")
    (261    ,    "YXEEB")
    (264    ,    "YXEET")
    (267    ,    "YXEEP")
    (270    ,    "YXX") # YXEEE

    (273    ,    "YXXK")
    (276    ,    "YXXM")
    (279    ,    "YXXB")
    (282    ,    "YXXT")
    (285    ,    "YXXP")
    (288    ,    "YXXE")

    (291    ,    "YXXEK")
    (294    ,    "YXXEM")
    (297    ,    "YXXEB")
    (300    ,    "YXXET")

BASE_10:float = 10.0

ToNotationString(Value:int)<transacts>: string = ToNotationString(Value*1.0)

ToNotationString(Value:float)<transacts>: string =
    for (Index->Definition:NotationDefinitions, Power:=Definition(0)*1.0, String:=Definition(1)):
        PowerValue:=Pow(BASE_10, Power)

        if (Value >= PowerValue):
            var bApply:logic = false

            if (NextDefinition:=NotationDefinitions[Index+1]):
                if:
                    NextPower:=NextDefinition(0)*1.0
                    NextPowerValue:=Pow(BASE_10, NextPower)
                    Value < NextPowerValue
                then:
                    set bApply = true
            else:
                set bApply = true
            
            if (bApply?, Number:=Value/PowerValue):
                return "{ToSmartString(Number, 2)}{String}"
        else:
            # Since array is supposed to be in order, it means the value is less than any notation, then no apply any
            return "{ToSmartString(Value, 2)}"

    "{ToSmartString(Value, 2)}"

