using { /Fortnite.com/Characters }
using { /Verse.org/Simulation }
using { /Verse.org/Concurrency }
using { /UnrealEngine.com/Temporary/Diagnostics }

<#
Gameplay stats that define a current player state
#>
player_stat := enum:
    Invulnerable
    # TODO probably we need to add one for UI @mbohorquez 07/18/2024 instead of using invulnerable to know if in UI
    MaxHealth
    MaxShield
    HealthRegen
    ShieldRegen
    # ...

ToString(Stat:player_stat)<computes>:string =
    case(Stat):
        player_stat.Invulnerable => "Invulnerable"
        player_stat.MaxHealth => "Max Health"
        player_stat.MaxShield => "Max Shield"
        player_stat.HealthRegen => "Health Regen"
        player_stat.ShieldRegen => "Shield Regen"
        # ... 

        #_ => "String Undefined - Update in ToString() function"

player_stat_changed_payload := struct:
    Player:player
    Stat:player_stat
    NewValue:float

<#
Has general gameplay data related to a player
#>
player_controller := class:
    Player:player

    var HealthRegenFrequency:float = 1.0
    var ShieldRegenFrequency:float = 1.0

    # Values for this player stat, if bool > 0 will be true
    var Stats:[player_stat]float = map{}
    
    # Number of requests to be invulnerable, defines if player is or no invulnerable
    var InvulnerableStackCount<private>:int = 0

    InvulnerableStackCountResetEvent<private>:event() = event(){}

    StatChangedEvent:event(player_stat_changed_payload) = event(player_stat_changed_payload){}

    EnteredRoomBeforeGoalEvent:event() = event(){}

    PlayerRemovedEvent:event() = event(){}

    RemoveStat(Stat:player_stat): void =
        if (CurrentValue:=Stats[Stat]):
            var NewMap:[player_stat]float = map{}
            for (Key -> Value : Stats, Key <> Stat):
                set NewMap = ConcatenateMaps(NewMap, map{Key => Value})
            set Stats = NewMap

            OnStatRemoved(Stat, CurrentValue)

    IsInState(Stat:player_stat)<decides><transacts>: void =
        GetStatValue[Stat] > 0.0

    GetStatValue(Stat:player_stat)<decides><transacts>: float =
        Stats[Stat]

    SetInState(Stat:player_stat, bNewState:logic, ?bForceUpdate:logic = false): void =
        NewValue:float = if (bNewState?) then 1.0 else 0.0
        UpdateStatValue(Stat, NewValue, ?bForceUpdate:=bForceUpdate)

    UpdateStatValue(Stat:player_stat, NewValue:float, ?bForceUpdate:logic = false): void =
        var MaybeCurrentValue:?float = false
        if (CurrentValue:=Stats[Stat]):
            set MaybeCurrentValue = option{CurrentValue}

        if (not bForceUpdate? and NewValue = MaybeCurrentValue?):
            return

        if (set Stats[Stat] = NewValue):
            OnStatChanged(Stat, NewValue, MaybeCurrentValue)

    AddStatValue(Stat:player_stat, AddValue:float): void =
        if (AddValue <> 0.0):
            CurrentValue:float = if (Value:=Stats[Stat]) then Value else 0.0
            UpdateStatValue(Stat, CurrentValue+AddValue)

    Initialize(Manager:player_controller_manager): void =
        set HealthRegenFrequency = Manager.HealthRegenFrequency
        set ShieldRegenFrequency = Manager.ShieldRegenFrequency

        # Initialize states that are needed
        for (Stat->Value:Manager.InitialStats):
            UpdateStatValue(Stat, Value, ?bForceUpdate:=true)

    # Called when player removed from match so all running routines are stopped
    OnPlayerRemoved(): void =
        PlayerRemovedEvent.Signal()

    OnSpawned(): void =
        # Force update states state, so fort character is updated correctly
        for (Stat->Value:Stats):
            OnStatChanged(Stat, Value, option{Value})

    OnEliminated(): void =
        # Player is by default Vulnerable
        ResetInvulnerableStackCount()

    # Call ONLY ONCE for making and ONLY ONCE for removing, this way we ensure vulnerable can be called anywhere and states will not be overriden
    MakeInvulnerable(bState:logic): void =
        set InvulnerableStackCount += if (bState?) then +1 else -1
        OnInvulnerableStackCountChanged()

    DelayMakeInvulnerable(bState:logic, Delay:float)<suspends>: void =
        if (Delay <= 0.0):
            MakeInvulnerable(bState)
        else:
            race:
                InvulnerableStackCountResetEvent.Await()

                block:
                    Sleep(Delay)
                    MakeInvulnerable(bState)

    ResetInvulnerableStackCount<private>(): void = 
        set InvulnerableStackCount = 0
        OnInvulnerableStackCountChanged()
        InvulnerableStackCountResetEvent.Signal()

    OnInvulnerableStackCountChanged<private>():void =
        SetInState(player_stat.Invulnerable, logic{InvulnerableStackCount > 0})

    # Called when a stat is removed
    OnStatRemoved<private>(Stat:player_stat, LastValue:float): void =
        NewValue:float = 0.0

        if (LastValue <> NewValue): # Only when it really will change
            StatChangedEvent.Signal(player_stat_changed_payload{Player:=Player, Stat:=Stat, NewValue:=NewValue})

            case(Stat):
                player_stat.Invulnerable =>
                    if:
                        LastValue > 0.0 # Was vulnerable
                        FortCharacter:=Player.GetFortCharacter[]
                    then:
                        FortCharacter.SetVulnerability(false)

                # ...
                # At the moment if any of the other stats are removed we do not update anything
                _ => {}

        
    # Called when a stat value changed, MaybeOldValue will not be defined when added for first time
    OnStatChanged<private>(Stat:player_stat, NewValue:float, MaybeOldValue:?float): void =
        StatChangedEvent.Signal(player_stat_changed_payload{Player:=Player, Stat:=Stat, NewValue:=NewValue})

        case(Stat):
            player_stat.Invulnerable =>
                if (FortCharacter:=Player.GetFortCharacter[]):
                    FortCharacter.SetVulnerability(logic{not IsInState[Stat]})

            player_stat.MaxHealth =>
                if (FortCharacter:=Player.GetFortCharacter[]):
                    FortCharacter.SetMaxHealth(NewValue)

            player_stat.MaxShield =>
                if (FortCharacter:=Player.GetFortCharacter[]):
                    FortCharacter.SetMaxShield(NewValue)

            player_stat.HealthRegen =>
                if:
                    not MaybeOldValue? or MaybeOldValue? <= 0.0
                then:
                    spawn {RunRegenBehavior(Stat, player_stat.MaxHealth, HealthRegenFrequency)}

            player_stat.ShieldRegen =>
                if:
                    not MaybeOldValue? or MaybeOldValue? <= 0.0
                then:
                    spawn {RunRegenBehavior(Stat, player_stat.MaxShield, ShieldRegenFrequency)}

            # ... Add other custom behaviors here

            #_ => {}

    RunRegenBehavior(Stat:player_stat, MaxStat:player_stat, Frequency:float)<suspends>: void =
        if (FortCharacter:=Player.GetFortCharacter[]):
            race:
                loop:
                    PlayerRemovedEvent.Await()
                    break

                loop:
                    StatChangedData:=StatChangedEvent.Await()
                    if:
                        StatChangedData.Stat = Stat
                        StatChangedData.NewValue <= 0.0
                    then:
                        break

                loop:
                    Sleep(Frequency)

                    if:
                        RegenValue:=GetStatValue[Stat]
                        RegenValue > 0.0
                        MaxValue:=GetStatValue[MaxStat]
                        FortCharacter.IsActive[] and not FortCharacter.IsDownButNotOut[]
                    then:
                        CurrentValue:float = case(Stat):
                            player_stat.HealthRegen => FortCharacter.GetHealth()
                            player_stat.ShieldRegen => FortCharacter.GetShield()
                            _ => 0.0

                        if (CurrentValue < MaxValue):
                            NewValue := CurrentValue + RegenValue
                            case(Stat):
                                player_stat.HealthRegen => FortCharacter.SetHealth(NewValue)
                                player_stat.ShieldRegen => FortCharacter.SetShield(NewValue)
                                _ => {}