using { /Fortnite.com/Devices }
using { /Fortnite.com/Characters }
using { /Verse.org/Simulation }
using { /Verse.org/Simulation/Tags }
using { /UnrealEngine.com/Temporary/Diagnostics }

log_player_controller_manager := class(log_channel){}
player_controller_manager_tag := class(tag){} # Tag to find the manager easily

<#
Controls player controllers in the game
#>
player_controller_manager := class(creative_device, stat_sender):
    Logger:log = log{Channel:=log_player_controller_manager}

    # Initial stat value the player begins with, for the moment not editable, if you want you need to create a struct
    InitialStats:[player_stat]float = map:
        player_stat.Invulnerable => 0.0
        # ...

    @editable
    HealthRegenFrequency:float = 3.0

    @editable
    ShieldRegenFrequency:float = 3.0

    var PlayerControllers:[player]player_controller = map{}

    @editable # Visually show that the player is Invulnerable
    InvulnerableVFXDevice:visual_effect_powerup_device = visual_effect_powerup_device{}

    @editable # Visually remove vfx when player is Invulnerable
    RemoveInvulnerableVFXDevice:visual_effect_powerup_device = visual_effect_powerup_device{}

    var StatsManager:?stats_manager = false

    PlayerRemovedEvent<private>:event(player) = event(player){}

    PlayerStatChangedEvent:event(player_stat_changed_payload) = event(player_stat_changed_payload){}

    OnBegin<override>()<suspends>:void =
        if(FoundStatsManager := GetStatManager[]):
            set StatsManager = option{FoundStatsManager}


    RegisterPlayer(Player:player):void =
        if:
            not PlayerControllers[Player]
            PlayerController:=player_controller{Player:=Player}
            set PlayerControllers[Player] = PlayerController
        then:
            spawn{PlayerController.AwaitPlayerEvents()}
            # After listening initialize player
            PlayerController.Initialize(Self)

    UnregisterPlayer(Player:player):void =
        if (PlayerController:=PlayerControllers[Player]):
            var NewMap:[player]player_controller = map{}
            for (KeyPlayer -> Data : PlayerControllers, KeyPlayer <> Player):
                set NewMap = ConcatenateMaps(NewMap, map{KeyPlayer => Data})
            set PlayerControllers = NewMap

            PlayerController.OnPlayerRemoved()
            PlayerRemovedEvent.Signal(Player)

    OnPlayerSpawned(Player:player): void =
        if (PlayerController:=PlayerControllers[Player]):
            PlayerController.OnSpawned()
            
        Data := array {
            stat_data{Player := Player, Category := stat_category.Run, Type := stat_type_new.Logic, ID := stat_id_new.Escaped, Value := 0.0}
        }
        if:
            SendStats[Data]

    OnPlayerEliminated(Player:player): void =
        if (PlayerController:=PlayerControllers[Player]):
            PlayerController.OnEliminated()
        Data := array {
            stat_data{Player := Player, Category := stat_category.General, Type := stat_type_new.Number, ID := stat_id_new.EscapeAttempts, Value := 1.0}
        }
        if:
            Manager := StatsManager?
            SendStats[Data]
            Manager.ResetRunStats[Player, false]

    OnPlayerEnteredRoomBeforeGoal(Player:player): void =
        if (PlayerController:=PlayerControllers[Player]):
            PlayerController.EnteredRoomBeforeGoalEvent.Signal()

    GetPlayerController(Player:player)<decides><transacts>:player_controller =
        PlayerControllers[Player]

    GetPlayerControllerAgent(Agent:agent)<decides><transacts>:player_controller =
        Player:=player[Agent]
        PlayerControllers[Player]
        
    #use to restore health and shield when changing classes
    ReconcilePlayerStats(Player:player):void=
        if:
            PlayerController := GetPlayerController[Player]
            FortCharacter := Player.GetFortCharacter[]

            MaxHealth := PlayerController.Stats[player_stat.MaxHealth]
            MaxShield := PlayerController.Stats[player_stat.MaxShield]

        then:
            FortCharacter.SetMaxHealth(MaxHealth)
            FortCharacter.SetMaxShield(MaxShield)

    (PlayerController:player_controller).AwaitPlayerEvents()<suspends>:void =
        race:
            loop:
                RemovedPlayer := PlayerRemovedEvent.Await()
                if (RemovedPlayer = PlayerController.Player):
                    break

            loop:
                StatChangedData := PlayerController.StatChangedEvent.Await()
                PlayerStatChangedEvent.Signal(StatChangedData)

                if (StatChangedData.Stat = player_stat.Invulnerable):
                    if (PlayerController.IsInState[player_stat.Invulnerable]):
                        InvulnerableVFXDevice.Pickup(PlayerController.Player)
                    else:
                        RemoveInvulnerableVFXDevice.Pickup(PlayerController.Player)

    SendStats<override>(Data:[]stat_data)<decides><transacts>:void = 
        if (Manager := StatsManager?):
            Manager.ReceiveData[Data]

    GetStatManager<override>()<decides><transacts>:stats_manager = 
        TaggedObjects := FindCreativeObjectsWithTag(stat_manager_tag{})
            var Objects:[]creative_object_interface=array{}
                for (Obj:TaggedObjects, TaggedObject:=creative_device_base[Obj]):
                    set Objects += array{TaggedObject}
                    #return Objects 
        Managers := for (Obj : TaggedObjects, Manager := stats_manager[Obj]):
            Manager
        if (Result := Managers[0]):
            Result
        else:
            Print("No stats manager found! returning empty instance")
            stats_manager{}