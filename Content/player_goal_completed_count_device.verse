
using { /Fortnite.com/Devices }
using { /Verse.org/Simulation }
using { /Verse.org/Simulation/Tags }
using { /UnrealEngine.com/Temporary/Diagnostics }

top_player_props_tag := class(tag) {}

player_goal_completed_count_device := class(creative_device):

    @editable
    PlayerHologram : player_reference_device = player_reference_device{}

    @editable
    PlayerPersistableData : player_persistable_data = player_persistable_data{}

    @editable
    DataBillboard : billboard_device = billboard_device{}

    var MaybeTopPlayer : ?player = false

    # Any additional props that should be shown/hidden when the top player is shown/hidden. These are assigned via tag.
    var TopPlayerProps : []creative_prop = array{}

    GoalReachedCountChangedEvent<public>: event(tuple(player, int)) = event(tuple(player, int)){}

    # --------------------------------------------------------------------------------
    OnBegin<override>()<suspends>:void=
        PropObjects := FindCreativeObjectsWithTag(top_player_props_tag{})
            var Objects: []creative_object_interface=array{}
                for(Obj:PropObjects, TaggedObject:=creative_device_base[Obj]):
                    set Objects += array{TaggedObject}

        set TopPlayerProps = for (PropObject : PropObjects, Prop := creative_prop[PropObject]) { Prop }

        GetPlayspace().PlayerAddedEvent().Subscribe(OnPlayerAdded)
        GetPlayspace().PlayerRemovedEvent().Subscribe(OnPlayerRemoved)

        UpdateTopPlayer()

    # --------------------------------------------------------------------------------
    OnPlayerAdded(Player:player):void=
        UpdateTopPlayer()

    # --------------------------------------------------------------------------------
    OnPlayerRemoved(Player:player):void=
        UpdateTopPlayer()

    # --------------------------------------------------------------------------------
    OnPlayerReachGoal(Player:player):void=
        if (NewCount := PlayerPersistableData.AddPlayerGoalReachedCount[Player]):
            UpdateTopPlayer()
            GoalReachedCountChangedEvent.Signal((Player, NewCount))
       
    # --------------------------------------------------------------------------------
    GetPlayerGoalReachedCount(Player:player): int =
        if (Value:=PlayerPersistableData.GetPlayerGoalReachedCount[Player]) then Value else 0

    # --------------------------------------------------------------------------------
    UpdateTopPlayer():void=
        set MaybeTopPlayer = false

        var MaybeTopPlayerState:?player_state = false
        for (Player : GetPlayspace().GetPlayers(), PlayerState := PlayerPersistableData.GetOrAddPlayerState[Player]):
            var bIsTop:logic = false

            # Only if has escaped at least once
            if (PlayerState.GoalReachedCount > 0):

                if (TopPlayerState:=MaybeTopPlayerState?):
                    if (PlayerState.GoalReachedCount > TopPlayerState.GoalReachedCount):
                        set bIsTop = true
                    else if (PlayerState.GoalReachedCount = TopPlayerState.GoalReachedCount):
                        if (LowestTimeData:=PlayerState.GetGoalLowestTime[]):
                            if (TopLowestTimeData:=TopPlayerState.GetGoalLowestTime[]):
                                VersionLowestTime:=LowestTimeData(0)
                                LowestTime:=LowestTimeData(1)

                                VersionTopLowestTime:=TopLowestTimeData(0)
                                TopLowestTime:=TopLowestTimeData(1)

                                if (LowestTime < TopLowestTime):
                                    set bIsTop = true
                                else if (LowestTime = TopLowestTime and VersionLowestTime >= VersionTopLowestTime): # Wins the one with the newer version
                                    set bIsTop = true
                            else:
                                set bIsTop = true
                else: 
                    set bIsTop = true

            if (bIsTop?):
                set MaybeTopPlayer = option{Player}
                set MaybeTopPlayerState = option{PlayerState}

        if (TopPlayer := MaybeTopPlayer?, TopPlayerState:=MaybeTopPlayerState?):
            ShowTopPlayer(TopPlayer, TopPlayerState)
        else:
            HideTopPlayer()

    # --------------------------------------------------------------------------------
    ShowTopPlayer(Player:player, PlayerState:player_state):void=
        PlayerHologram.Enable()
        PlayerHologram.Register(Player)

        var StringData:string = "Escapes: {PlayerState.GoalReachedCount}"
        if (LowestTimeData:=PlayerState.GetGoalLowestTime[]):
            set StringData += "\nShortest Time: {TimeInSecondsToString(LowestTimeData(1))}"

        DataBillboard.SetText(StringToMessage(StringData))
        DataBillboard.ShowText()

        for (Prop : TopPlayerProps):
            Prop.Show()

    # --------------------------------------------------------------------------------
    HideTopPlayer():void=
        PlayerHologram.Disable()
        DataBillboard.HideText()

        for (Prop : TopPlayerProps):
            Prop.Hide()