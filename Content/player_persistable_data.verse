using { /Fortnite.com/Devices }
using { /Verse.org/Simulation }
using { /UnrealEngine.com/Temporary/Diagnostics }

log_player_persistable_data := class(log_channel){}

<#  
*** UPDATE THIS VERSION ALWAYS When a new version of the island is published ***
*** UPDATE Version Default value in player_state ***
Add notes of your version changes in the description if needed
#>
CURRENT_VERSION:int = 4

<# 
PLAYER STATE VERSION DESCRIPTIONS
Add here the description of what each version means
...

Version 0: Initial data the player had since the beginning of the game
Version 1: There was an exploid with caffenaited revival (creator portal Version 5), it includes the fixes for that (`LowestGoalTime` deprecated and now using `ArriveGoalLowestTimes`)
Version 2: Timer paused in the reward selection was removed / Rebalancing data
Version 3: Guards in last room had been submitted wrong
Version 4: Prestige system activated, ShowCrown added | Currencies deprecated, now CurrencyValues | Diamonds in inventory moved to player_state (Players data updated)

...
#>

<# 
Data that will be saved and persistable for the player
... Add here any custom data in your game ...
#>
player_state := class<final><persistable>():
    Version:int = 4 # You need to UPDATE this value manually when a new version is published: Same value as CURRENT_VERSION /Variable is not set directly because crash

    LowestGoalTime:int = -1 # DEPRECATED, time used in Version <= 0
    ArriveGoalLowestTimes:[int]int = map{} # [Version]LowestTime

    GoalReachedCount:int = 0

    #Currencies:[currency]float = map{currency.Gold => 0.0} # DEPRECATED, currencies used in Version <= 5
    CurrencyValues:[int]float = map{} # [CurrencyIndex]Value

    #CurrencyUpgrades:[currency][]currency_upgrade = map{} # This uses deprecated currency enum, but we will leave it since at this moment there is only gold upgrades

    QuestIndex:?int = option{0} # A false value means that the player has finished all quests
    QuestProgress:float = 0.0 # Stores the progress of the current quest
    
    #CurrencyMaxout : []tuple(currency_upgrade_id, int) = array{} # DEPRECATED, this was never used, someone probably added it by mistake (cannot be removed because island published)

    PrestigeLevel:int=0
    MaybeShowCrown:?logic=false # If the player has show crown unlocked and what value

    # ...

var PlayerStates:weak_map(player, player_state) = map{}

<# 
Constructor for the player state data
... Add here any custom data in your game ...
#>
MakePlayerState<constructor>(Src:player_state)<transacts> := player_state:
    Version := Src.Version

    LowestGoalTime := Src.LowestGoalTime
    ArriveGoalLowestTimes := Src.ArriveGoalLowestTimes

    GoalReachedCount := Src.GoalReachedCount

    #Currencies := Src.Currencies
    #CurrencyValues := Src.CurrencyValues

    #CurrencyUpgrades := Src.CurrencyUpgrades

    QuestIndex := Src.QuestIndex
    QuestProgress := Src.QuestProgress

    PrestigeLevel:=Src.PrestigeLevel
    MaybeShowCrown:=Src.MaybeShowCrown

    # ...

<# Player State Extensions #>
# Returns the lowest time this player has made as (Version, Time)
(PlayerState:player_state).GetGoalLowestTime()<decides><transacts>: tuple(int, int) =
    var MaybeLowestTimeData:?tuple(int, int) = false

    var LowestTimeData:tuple(int, int) = (-1, -1) # (Version, Time)
    for (Version->Time:PlayerState.ArriveGoalLowestTimes):
        if (Time > 0):
            CurrentVersion:=LowestTimeData(0)
            CurrentLowestTime:=LowestTimeData(1)
            
            if (CurrentLowestTime <= 0):
                set LowestTimeData = (Version, Time)
            else if (Time < CurrentLowestTime):
                set LowestTimeData = (Version, Time)
            else if (Time = CurrentLowestTime and Version > CurrentVersion): # Priority to the newer version
                set LowestTimeData = (Version, Time)

    if (LowestTimeData(1) > 0):
        set MaybeLowestTimeData = option{LowestTimeData}

    MaybeLowestTimeData?

<#
Controls persitable data per player
#>
player_persistable_data := class(creative_device, stat_sender):
    Logger:log = log{Channel:=log_player_persistable_data}

    @editable
    SavePointOnPlayerLoadedTrigger:trigger_device = trigger_device{}

    SavePointPlayerLoadedEvent:event(player) = event(player){}

    var StatsManager:?stats_manager = false

    OnBegin<override>()<suspends>:void =
        if(FoundStatsManager := GetStatManager[]):
            set StatsManager = option{FoundStatsManager}
        loop:
            MaybeAgent:=SavePointOnPlayerLoadedTrigger.TriggeredEvent.Await()
            if (Player:=player[MaybeAgent?]):
                SavePointPlayerLoadedEvent.Signal(Player)

    GetOrAddPlayerState(Player:player)<decides><transacts>:player_state =
        var MaybePlayerState:?player_state = false
        if:
            not PlayerStates[Player]
            set PlayerStates[Player] = player_state{}
            Logger.Print("Player State Added")
        if:
            PlayerState := PlayerStates[Player]
            set MaybePlayerState = option{PlayerState}
            Logger.Print("Player State Found")
        return MaybePlayerState?

    GetPlayerState(Player:player)<decides><transacts>: player_state = PlayerStates[Player]

    GetMaybePlayerState(Player:player): ?player_state =
        if (PlayerState:=GetPlayerState[Player]):
            return option{PlayerState}
        return false

    ClearPlayerState(Player:player): void =
        if:
            PlayerState := PlayerStates[Player]
            set PlayerStates[Player] = player_state:
                Version:=PlayerState.Version

    # Called to make any data adjustments when player starts
    InitializePlayerState(Player:player, PlayerState:player_state, CurrencyManager:clash_currency_manager): void =
        Logger.Print("Current player version {PlayerState.Version}")

        var NewPlayerState:player_state = PlayerState
        var bNeedsUpdate:logic = false

        # LowestTime data has not been updated
        if (PlayerState.Version <= 0):
            if (NewPlayerState.LowestGoalTime >= 0):
                var NewArriveGoalLowestTimes:[int]int = NewPlayerState.ArriveGoalLowestTimes
                var LowestTime:int = NewPlayerState.LowestGoalTime

                # Do not change this value, it is the minimum time we had at that point and any time under it was taken as invalid
                RemoveTimesUnder:int = 374 # 374s = 6:14
                # Do not change this value, version 0 is the one before all the changes
                FixVersion:int = 0

                if (LowestTime < RemoveTimesUnder):
                    Logger.Print("Player v0 LowestTime {LowestTime} removed")
                    set LowestTime = VALUE_NONE

                # Update the NEW `ArriveGoalLowestTimes` for the fix version
                if:
                    LowestTime > 0
                    set NewArriveGoalLowestTimes[FixVersion] = LowestTime
                then:
                    Logger.Print("Player v0 LowestTime {LowestTime} kept")
                    
                set NewPlayerState = player_state:
                    MakePlayerState<constructor>(NewPlayerState)
                    ArriveGoalLowestTimes:=NewArriveGoalLowestTimes

            set bNeedsUpdate = true
            Logger.Print("Player data for version <=0 fixed (LowestTime moved to ArriveGoalLowestTimes)")

        # Deprecated Currencies was used
        if (PlayerState.Version <= 3):
            var NewCurrencyValues:[int]float = map{}

            <#for (DeprecatedCurrency->Value:NewPlayerState.Currencies):
                Index:=GetIndexCurrency(ToCurrencyId(DeprecatedCurrency))
                if (Index >= 0, set NewCurrencyValues[Index] = Value):
                    Logger.Print("Player deprecated Currency {Index} moved")#>

            set NewPlayerState = player_state:
                MakePlayerState<constructor>(NewPlayerState)
                CurrencyValues:=NewCurrencyValues
            set bNeedsUpdate = true
            Logger.Print("Player data for version <=4 fixed (Curriencies moved to CurrencyValues)")

        # Update
        if:
            bNeedsUpdate?
            set PlayerStates[Player] = player_state:
                MakePlayerState<constructor>(NewPlayerState)
                Version:=CURRENT_VERSION
        then:
            Logger.Print("Player data initialized and updated - Version {CURRENT_VERSION}")

        # After data updated try fix inventory, just in case save point was called before
        #CurrencyManager.FixPlayerLegacyInventory(Player)

    # ARRIVE GOAL LOWEST TIME
    GetPlayerArriveGoalLowestTimes(Player:player)<decides><transacts>: [int]int =
        PlayerStates[Player].ArriveGoalLowestTimes

    TryRegisterArriveGoalLowestTime(Player:player, Time:int)<decides><transacts>: int =
        var MaybeTime:?int = false

        if (Time > 0, PlayerState := PlayerStates[Player]):
            var LowestTime:int = if (VersionLowestTime:=PlayerState.ArriveGoalLowestTimes[CURRENT_VERSION]) then VersionLowestTime else -1
            if (LowestTime <= 0 or Time < LowestTime):
                var NewArriveGoalLowestTimes:[int]int = PlayerState.ArriveGoalLowestTimes

                if (set NewArriveGoalLowestTimes[CURRENT_VERSION] = Time):
                    if:
                        set PlayerStates[Player] = player_state:
                            MakePlayerState<constructor>(PlayerState)
                            ArriveGoalLowestTimes:=NewArriveGoalLowestTimes
                    then:
                        Logger.Print("Arrive Goal Lowest Time {Time} registered for version {CURRENT_VERSION}")
                        set MaybeTime = option{Time}
                        set LowestTime = Time

            Data := array {
            stat_data{Player := Player, Category := stat_category.General, Type := stat_type_new.Number, ID := stat_id_new.ShortestEscape, Value := LowestTime*1.0, CalculationMethod := stat_calculation_method.Lowest, InsertionMethod := stat_insertion_method.Lowest}
            }
            if(SendStats[Data]):
        MaybeTime?

    HasPlayerArriveGoalLowestTimes(Player:player)<decides><transacts>: void =
        PlayerStates[Player].ArriveGoalLowestTimes.Length > 0

    ResetArriveGoalLowestTimes(Player:player)<decides><transacts>: void =
        PlayerState := PlayerStates[Player]
        PlayerState.ArriveGoalLowestTimes.Length > 0 or PlayerState.LowestGoalTime >= 0
        
        set PlayerStates[Player] = player_state:
            MakePlayerState<constructor>(PlayerState)
            LowestGoalTime:=VALUE_NONE
            ArriveGoalLowestTimes:= map{}

    # Returns the player goal lowest time in any version if registered tuple(Version, Time)
    GetPlayerGoalLowestTime(Player:player)<decides><transacts>: tuple(int, int) =
        PlayerStates[Player].GetGoalLowestTime[]
    
    # GOAL REACHED COUNT
    GetPlayerGoalReachedCount(Player:player)<decides><transacts>: int =
        PlayerStates[Player].GoalReachedCount

    # Adds to player goal reached count and returns the new count
    AddPlayerGoalReachedCount(Player:player, ?CountToAdd:int=1)<decides><transacts>: int =
        PlayerState := PlayerStates[Player]
        CurrentCount := PlayerState.GoalReachedCount
        NewCount := CurrentCount + CountToAdd

        set PlayerStates[Player] = player_state:
            MakePlayerState<constructor>(PlayerState)
            GoalReachedCount := CurrentCount + CountToAdd

        NewCount

    # CURRENCY VALUES
    <#GetPlayerCurrencyValues(Player:player)<decides><transacts>: [int]float =
        PlayerStates[Player].CurrencyValues

    GetPlayerCurrency(Player:player, Currency:currency_id)<transacts>: float =
        if (Index:=GetIndexCurrency(Currency), Value:=PlayerStates[Player].CurrencyValues[Index]) then Value else 0.0

    UpdatePlayerCurrencyValues<private>(Player:player, CurrencyValues:[int]float)<decides><transacts>: void =
        PlayerState := PlayerStates[Player]
        set PlayerStates[Player] = player_state:
            MakePlayerState<constructor>(PlayerState)
            CurrencyValues := CurrencyValues

    UpdatePlayerCurrency(Player:player, Currency:currency_id, NewValue:float, ?CurrencyIcon: currency_source_id = currency_source_id.Kill)<decides><transacts>: void =
        var bSuccess:logic = false
        if (NewValue >= 0.0, var CurrencyValues:[int]float = PlayerStates[Player].CurrencyValues):
            var bShouldUpdate:logic = true
            Index:=GetIndexCurrency(Currency)

            if (Index >= 0, CurrencyValue := CurrencyValues[Index], CurrencyValue = NewValue):
                set bShouldUpdate = false

            if:
                bShouldUpdate?
                set CurrencyValues[Index] = NewValue
                UpdatePlayerCurrencyValues[Player, CurrencyValues]
            then:
                Logger.Print("Currency {Currency} updated with value {NewValue}")
                set bSuccess = true
                
        bSuccess?#>

    # CURRENCY UPGRADES
    <#GetPlayerCurrencyUpgrades(Player:player)<decides><transacts>: [currency][]currency_upgrade =
        PlayerStates[Player].CurrencyUpgrades

    UpdatePlayerCurrencyUpgrades(Player:player, CurrencyUpgrades:[currency][]currency_upgrade)<decides><transacts>: void =
        PlayerState := PlayerStates[Player]
        set PlayerStates[Player] = player_state:
            MakePlayerState<constructor>(PlayerState)
            CurrencyUpgrades := CurrencyUpgrades

    GetPlayerCurrencyUpgrades(Player:player, Currency:currency_id)<decides><transacts>: []currency_upgrade =
        DeprecatedCurrency:=ToDeprecatedCurrency[Currency]
        PlayerStates[Player].CurrencyUpgrades[DeprecatedCurrency]

    GetMaybePlayerCurrencyUpgrade(Player:player, Currency:currency_id, UpgradeId:currency_upgrade_id)<transacts>:?currency_upgrade =
        var MaybePlayerCurrencyUpgrade:?currency_upgrade = false

        if (PlayerCurrencyUpgrades := GetPlayerCurrencyUpgrades[Player, Currency]):
            for (CurrencyUpgrade:PlayerCurrencyUpgrades):
                if (CurrencyUpgrade.Id = UpgradeId):
                    set MaybePlayerCurrencyUpgrade = option{CurrencyUpgrade}

        MaybePlayerCurrencyUpgrade#>

    # QUESTS
    SetPlayerQuestIndex(Player:player, Index:int)<decides><transacts>:void=
        PlayerState := PlayerStates[Player]
        set PlayerStates[Player] = player_state:
            MakePlayerState<constructor>(PlayerState)
            QuestIndex := option{Index}

    GetPlayerQuestProgress(Player:player)<decides><transacts>:float =
        PlayerStates[Player].QuestProgress

    SetPlayerQuestProgress(Player:player, Progress:float)<decides><transacts>:void=
        PlayerState := PlayerStates[Player]
        set PlayerStates[Player] = player_state:
            MakePlayerState<constructor>(PlayerState)
            QuestProgress := Progress

    SetPlayerFinishedQuests(Player:player)<decides><transacts>:void=
        PlayerState := PlayerStates[Player]
        set PlayerStates[Player] = player_state:
            MakePlayerState<constructor>(PlayerState)
            QuestIndex := false

    # PRESTIGE
    SetPlayerPrestigeLevel(Player:player, NewPrestigeLevel : int)<decides><transacts>:void=
        PlayerState := PlayerStates[Player]
        set PlayerStates[Player] = player_state:
          MakePlayerState<constructor>(PlayerState)
          PrestigeLevel:=NewPrestigeLevel
        Data := array {
            stat_data{Player := Player, Category := stat_category.General, Type := stat_type_new.Number, ID := stat_id_new.PrestigeLevel, Value := NewPrestigeLevel*1.0, CalculationMethod := stat_calculation_method.Newest}
            }
        SendStats[Data]


    GetPlayerPrestigeLevel(Player:player)<decides><transacts>:int =
        PlayerStates[Player].PrestigeLevel

    # SHOW CROWN
    GetMaybePlayerShowCrown(Player:player)<transacts>:?logic =
        if (PlayerState:=PlayerStates[Player]):
            PlayerState.MaybeShowCrown
        else:
            false

    SetPlayerShowCrown(Player:player, bShow:logic)<decides><transacts>:void =
        PlayerState := PlayerStates[Player]
        set PlayerStates[Player] = player_state:
            MakePlayerState<constructor>(PlayerState)
            MaybeShowCrown:=option{bShow}

    ResetPlayerShowCrown(Player:player)<decides><transacts>:void =
        PlayerState := PlayerStates[Player]
        set PlayerStates[Player] = player_state:
            MakePlayerState<constructor>(PlayerState)
            MaybeShowCrown:=false

    # Stats
    SendStats<override>(Data:[]stat_data)<decides><transacts>:void = 
        if (Manager := StatsManager?):
            Manager.ReceiveData[Data]

    GetStatManager<override>()<decides><transacts>:stats_manager = 
        TaggedObjects := FindCreativeObjectsWithTag(stat_manager_tag{})
            var Objects:[]creative_object_interface=array{}
                for (Obj:TaggedObjects, TaggedObject:=creative_device_base[Obj]):
                    set Objects += array{TaggedObject}
                    #return Objects 
        Managers := for (Obj : TaggedObjects, Manager := stats_manager[Obj]):
            Manager
        if (Result := Managers[0]):
            Result
        else:
            Print("No stats manager found! returning empty instance")
            stats_manager{}
