using { /Fortnite.com/Devices }
using { /Verse.org/Simulation }
using { /UnrealEngine.com/Temporary/Diagnostics }
using { /Fortnite.com/Characters }
using { /Fortnite.com/Game }
using { /Verse.org/Colors }
using { /UnrealEngine.com/Temporary/SpatialMath }


player_session_data := class:
    # @editable
    # var RewardRolls:int=0

    # @editable
    # var GuestRewardMultiplier:float=0.0

    # @editable
    # var KillStreakCount:int=0

    # @editable
    # var DiamondWizMultiplier:float=1.0

    # @editable
    # var DiamondWizDiamondMultiplier:float=0.0

    # @editable
    # var DiamondWizAlreadyInteracted:logic=false

    # @editable
    # var KillStreakMultiplier:float=0.0
    
    var Team:team_id

# A Verse-authored creative device that can be placed in a level
player_temp_data := class(creative_device):

    var PlayersSessionData:[player]player_session_data=map{}

    # @editable
    # MaybeRoomManager:?room_manager=false

    # @editable
    # PlayerEliminationManager:elimination_manager_device=elimination_manager_device{}

    # var StatsManager:?stats_manager = false

    OnPlayerDamaged : event(damage_result) = event(damage_result){}

    OnBegin<override>()<suspends>:void =

        # if(FoundStatsManager := GetStatManager[]):
        #     set StatsManager = option{FoundStatsManager}

        Players := GetPlayspace().GetPlayers()

        for (Player : Players):
            RegisterNewPlayer(Player)

        GetPlayspace().PlayerAddedEvent().Subscribe(RegisterNewPlayer)                        
        GetPlayspace().PlayerRemovedEvent().Subscribe(RemovePlayer)
        
        # if(RoomManager:=room_manager[MaybeRoomManager?]):
        #      spawn {CheckPlayerState(RoomManager)}


    GetPlayerSessionData(Player : player)<transacts><decides>:player_session_data=
        PlayersSessionData[Player]

    GetPlayerTeam(Player:player)<decides><transacts>:team_id = 
        Data := GetPlayerSessionData[Player]
        Data.Team


    UpdatePlayerRewardRolls<public>(Player : player):void= {}
    #     if(SessionData:= PlayersSessionData[Player]):
    #         set SessionData.RewardRolls+=1
    #         Data := array {
    #             stat_data{Player := Player, Category := stat_category.General, Type := stat_type_new.Number, ID := stat_id_new.Rerolls, Value := 1.0, InsertionMethod := stat_insertion_method.Highest}
    #             stat_data{Player := Player, Category := stat_category.Run, Type := stat_type_new.Number, ID := stat_id_new.Rerolls, Value := 1.0}
    #         }
    #         if(SendStats[Data]):

    IncreasePlayerGuestMultiplier<public>(Player : player, Value:float):void={}    
    #     if(SessionData:= PlayersSessionData[Player]):
    #         set SessionData.GuestRewardMultiplier+=Value

    UpdatePlayerKillStreakCount<public>(Player : player):int= 0
    #     if(SessionData:= PlayersSessionData[Player]):
    #         set SessionData.KillStreakCount +=1

    #         Data := array {
    #             stat_data{Player := Player, Category := stat_category.General, Type := stat_type_new.Number, ID := stat_id_new.HighestEliminationStreak, Value := 1.0, InsertionMethod := stat_insertion_method.Highest}
    #             stat_data{Player := Player, Category := stat_category.Run, Type := stat_type_new.Number, ID := stat_id_new.HighestEliminationStreak, Value := 1.0, InsertionMethod := stat_insertion_method.Highest}
    #         }
    #         if(SendStats[Data]):

    #         if(NewKillCount := PlayersSessionData[Player].KillStreakCount):
    #             return NewKillCount
    #     # Return -1 if fails.
    #     return -1 

    # # jcarrero - Check with Game Design the static value of the increment multiplier    
    UpdateDiamondWizMultiplier<public>(Player : player, ShouldAdd : logic):float= 0.0    
    #     if(SessionData:= PlayersSessionData[Player]):
    #         if(ShouldAdd?):
    #             # jcarrero - Talk with Game Design about the max allowed number to multiply
    #             if(SessionData.DiamondWizMultiplier < 99.0 ):
    #                 set SessionData.DiamondWizMultiplier+=0.1  
    #         else:
    #             if(SessionData.DiamondWizMultiplier > 1.0 ):
    #                 set SessionData.DiamondWizMultiplier-=0.1  
    #         return SessionData.DiamondWizMultiplier
    #     else:
    #         return -1.0
        
    UpdateDiamondWizDiamondMultiplier<public>(Player : player, ShouldAdd : logic):float= 0.0
    #     if(SessionData:= PlayersSessionData[Player]):
    #         if(ShouldAdd?):
    #             set SessionData.DiamondWizDiamondMultiplier+=1.0  
    #         else:
    #             set SessionData.DiamondWizDiamondMultiplier-=1.0  
    #         return SessionData.DiamondWizDiamondMultiplier
    #     else:
    #         return -1.0   

    UpdateDiamondWizInteractionStatus<public>(Player : player, InteractionStatus : logic):logic= false 
    #     if(SessionData:= PlayersSessionData[Player]):
    #         set SessionData.DiamondWizAlreadyInteracted = InteractionStatus 
    #         return SessionData.DiamondWizAlreadyInteracted
    #     else:
    #         return false         

    SetPlayerKillStreakMultiplier<public>(Player : player, NewMultiplierClamp:float):void= {}   
    #     if(SessionData:= PlayersSessionData[Player]):
    #         set SessionData.KillStreakMultiplier=NewMultiplierClamp


    GetPlayerRewardRollsCount<public>(Player : player)<transacts><decides>:int= 0
    #     PlayersSessionData[Player].RewardRolls
        
    GetPlayerGuestRewardMultiplier<public>(Player : player)<transacts><decides>:float= 0.0
    #     PlayersSessionData[Player].GuestRewardMultiplier

    GetPlayerKillStreakCount<public>(Player : player)<transacts><decides>:int= 0
    #     PlayersSessionData[Player].KillStreakCount

    GetDiamondWizMultiplier<public>(Player : player)<transacts><decides>:float= 0.0
    #     PlayersSessionData[Player].DiamondWizMultiplier 
            
    GetDiamondWizDiamondMultiplier<public>(Player : player)<transacts><decides>:float= 0.0
    #     PlayersSessionData[Player].DiamondWizDiamondMultiplier  

    GetDiamondWizInteractionStatus<public>(Player : player):logic= false    
    #     if(SessionData:= PlayersSessionData[Player]):
    #         return SessionData.DiamondWizAlreadyInteracted
    #     else:
    #         return false      

    GetPlayerKillStreakMultiplier<public>(Player : player)<transacts><decides>:float=  0.0  
    #     PlayersSessionData[Player].KillStreakMultiplier

    RegisterNewPlayer(Player : player):void =
        if:
            not PlayersSessionData[Player]
            PlayerTeam := GetAgentTeam(Player)
            set PlayersSessionData[Player] = player_session_data{Team := PlayerTeam}

                # Bind player events
                FortCharacter := Player.GetFortCharacter[]
        then:
            FortCharacter.DamagedEvent().Subscribe(CharacterDamaged)
            FortCharacter.DamagedShieldEvent().Subscribe(CharacterDamaged)

    # ResetPlayerData(Player : player):void=
    #     if(PlayersSessionData[Player]):
    #         if:
    #             set PlayersSessionData[Player] = player_session_data{}    
    #         then:   
    #             ResetDiamondWizInteractionStatus(Player)

    ResetPlayerRewardRolls(Player:player):void= {}
        # if(set PlayersSessionData[Player].RewardRolls = 0){}

    ResetPlayerGuestMultiplier(Player:player):void= {}
    #     if(set PlayersSessionData[Player].GuestRewardMultiplier = 0.0){}

    ResetKillStreak(Player : player):void= {}
    #     if(set PlayersSessionData[Player].KillStreakCount = 0):
    #         #Print(PlayerNameToMessage(Player), ?Color := NamedColors.Green, ?Duration := 10.0)
    #         if(KillCount := PlayersSessionData[Player].KillStreakCount):
    #             #HH2Print("Kills: {KillCount}", ?TextColor := NamedColors.Green)
    #     if(set PlayersSessionData[Player].KillStreakCount = 0){}
    #     if(set PlayersSessionData[Player].KillStreakMultiplier = 0.0){}
    #     Data := array {
    #         stat_data{Player := Player, Category := stat_category.General, Type := stat_type_new.Number, ID := stat_id_new.HighestEliminationStreak, Value := 0.0, CalculationMethod := stat_calculation_method.Newest, InsertionMethod := stat_insertion_method.Highest}
    #             stat_data{Player := Player, Category := stat_category.Run, Type := stat_type_new.Number, ID := stat_id_new.HighestEliminationStreak, Value := 0.0, CalculationMethod := stat_calculation_method.Newest, InsertionMethod := stat_insertion_method.Highest}
    #         }
    #     if(SendStats[Data]):


    ResetDiamondWizMultiplier(Player : player):void= {}
    #     if(set PlayersSessionData[Player].DiamondWizMultiplier = 1.0){}         
        
    ResetDiamondWizDiamondMultiplier(Player : player):void= {}
    #     if(set PlayersSessionData[Player].DiamondWizDiamondMultiplier = 0.0){}   

    ResetDiamondWizInteractionStatus(Player : player):void= {}
    #     if(set PlayersSessionData[Player].DiamondWizAlreadyInteracted = false){}  
            
    RemovePlayer(Player : player):void=
        #Remove player from map
        if (ActualPlayer := PlayersSessionData[Player]):

            var TempAllPlayersMap:[player]player_session_data = map{}

            for (Key -> Value : PlayersSessionData, Key <> Player):
                set TempAllPlayersMap = ConcatenateMaps (TempAllPlayersMap, map{Key => Value})

            set PlayersSessionData = TempAllPlayersMap

    # CheckPlayerState(RoomManager : room_manager)<suspends>:void=

    #     sync:

    #         loop:
    #             GoalPlayer:= RoomManager.OnGoalCompleted.Await()
    #             #Print("On Player Goal Complete")
    #             ResetPlayerData(GoalPlayer)
    #         loop:
    #             BeforeGoalPlayer:= RoomManager.OnPlayerEnteredRoomBeforeGoal.Await()
    #             #Print("On Player Enter Room Before Goal: Main Room")
    #         loop:
    #             EliminatedAgent:= PlayerEliminationManager.EliminatedEvent.Await()
    #             #Print("On Player Eliminated")
    #             if(EliminatedPlayer:=player[EliminatedAgent]):
    #                 #Data to Reset
    #                 ResetPlayerRewardRolls(EliminatedPlayer)

    # # -------- Binded Events --------
    CharacterDamaged(DamageResult : damage_result):void=
        OnPlayerDamaged.Signal(DamageResult)

    # SendStats<override>(Data:[]stat_data)<decides><transacts>:void = 
    #     if (Manager := StatsManager?):
    #         Manager.ReceiveData[Data]

    # GetStatManager<override>()<decides><transacts>:stats_manager = 
    #     TaggedObjects := FindCreativeObjectsWithTag(stat_manager_tag{})
    #         var Objects:[]creative_object_interface=array{}
    #             for (Obj:TaggedObjects, TaggedObject:=creative_device_base[Obj]):
    #                 set Objects += array{TaggedObject}
    #                 #return Objects 
    #     Managers := for (Obj : TaggedObjects, Manager := stats_manager[Obj]):
    #         Manager
    #     if (Result := Managers[0]):
    #         Result
    #     else:
    #         Print("No stats manager found! returning empty instance")
    #         stats_manager{}
