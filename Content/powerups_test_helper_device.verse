
using { /Fortnite.com/Devices }
using { /Verse.org/Simulation }
using { /UnrealEngine.com/Temporary/Diagnostics }

test_helper_powerups_device := class(creative_device):

    @editable
    PowerupManager:powerup_manager_device=powerup_manager_device{}

    @editable
    var UpgradeLevel:int=1

    @editable
    GiveButtons:[]button_device=array{}

    @editable
    IncreaseUpgradeLevelButton:button_device=button_device{}

    @editable
    MaxIncreaseLevel:int=10

    @editable
    RemoveButton:button_device=button_device{}

    OnBegin<override>()<suspends>:void=
        Print("Waiting for devices to Start...",?Duration:= 2.0)
        Sleep(2.0)
         Print("Getting Powerup Devices...")
        var Index:int=0
        for(GiveButton:GiveButtons):
            GiveButton.InteractedWithEvent.SubscribeAgent(OnButtonInteracted, Index)
            #Set Button Text
            if(TargetPowerup:=PowerupManager.AvailablePowerUps[Index]):
                GiveButton.SetInteractionText(StringToMessage(TargetPowerup.PowerupDisplayName))
            set Index+=1
        IncreaseUpgradeLevelButton.InteractedWithEvent.Subscribe(OnButtonInteracted_2)
        RemoveButton.InteractedWithEvent.Subscribe(OnButtonInteracted_3)

    OnButtonInteracted(Agent:agent, Index:int):void=
        if(Player:=player[Agent]):
           if(TargetPowerup:=PowerupManager.AvailablePowerUps[Index]):
            Print("Giving: {TargetPowerup.PowerupDisplayName}")
            PID:=TargetPowerup.PowerupID
            PowerupManager.GivePowerUpToPlayer(Player, PID, UpgradeLevel)

    OnButtonInteracted_2(Agent:agent):void=
        if(Player:=player[Agent]):
            set UpgradeLevel+=1
            if(UpgradeLevel > MaxIncreaseLevel):
                set UpgradeLevel = 0   
            Print("New Upgrade level: {UpgradeLevel}")
            
    OnButtonInteracted_3(Agent:agent):void=
        Print("Removing all powerups for this player")
        if(Player:=player[Agent]):
            for(PowerUp:PowerupManager.AvailablePowerUps):
                Print("Removing: {PowerUp.PowerupDisplayName}")
                PowerupManager.RemovePowerupFromPlayer(Player,PowerUp.PowerupID)


# Code to be able to pass more arguments to event subscriptions. Works for events that send an agent and that don't send anything.
# (Listenable : listenable(agent)).SubscribeAgent(OutputFunc : tuple(agent, t)->void, ExtraData : t where t:type) : cancelable =
#     Wrapper := wrapper_agent(t){ExtraData := ExtraData, OutputFunc := OutputFunc}
#     Listenable.Subscribe(Wrapper.InputFunc)
    
# wrapper_agent(t : type) := class():
#     ExtraData : t;
#     OutputFunc : tuple(agent, t) -> void
#     InputFunc(Agent : agent):void = OutputFunc(Agent, ExtraData)
