using { /Verse.org/Simulation/Tags }
using { /Verse.org/Simulation }
using { /UnrealEngine.com/Temporary/UI }
using { /UnrealEngine.com/Temporary/Diagnostics }
using { /Fortnite.com/UI }

#=======================================================================================================================
#Classes

    powerup_upgrade_data:=class:
        var bIsPowerUpOwned:logic=false
        var MaybeIconWidget:?texture_block=false
        var MaybeUpgradeTextBlock:?text_block=false
        var MaybePowerupStackBox:?stack_box=false

    (PowerUpUpgradeData:powerup_upgrade_data).SetWidgets(NewTexture:texture_block, NewTextBlock:text_block, NewPowerupStackBox:stack_box):void=
        set PowerUpUpgradeData.MaybeIconWidget = option{NewTexture}
        set PowerUpUpgradeData.MaybeUpgradeTextBlock = option{NewTextBlock}
        set PowerUpUpgradeData.MaybePowerupStackBox = option{NewPowerupStackBox}

    (PowerUpUpgradeData:powerup_upgrade_data).SetUpgradeText(NewUpgrade:string):void=
        if(ValidTextBlock:=PowerUpUpgradeData.MaybeUpgradeTextBlock?):
            ValidTextBlock.SetText(StringToMessage("{NewUpgrade}"))

    powerup_player_data:=class:
        var Data:[powerup_id]powerup_upgrade_data=map{}
        var Canvas:canvas=canvas{}
        var StackBox:stack_box=stack_box{Orientation:=orientation.Horizontal}
        var bIsCanvasSet:logic=false
        var MaybeTask:?cancelable=false

    (PowerUpDataArray:powerup_player_data).AddData(PowerupID:powerup_id, NewData:powerup_upgrade_data):void=
        if:
            set NewData.bIsPowerUpOwned = true
            set PowerUpDataArray.Data[PowerupID] = NewData

    (PowerUpDataArray:powerup_player_data).GetData(PowerupID:powerup_id):powerup_upgrade_data=
        if(FoundData:=PowerUpDataArray.Data[PowerupID]):
            return FoundData
        else:
            NewData:=powerup_upgrade_data{}
            PowerUpDataArray.AddData(PowerupID, NewData)
            return NewData

    (PowerUpDataArray:powerup_player_data).ResetData(PowerupID:powerup_id):void=
        if:
            set PowerUpDataArray.Data[PowerupID] = powerup_upgrade_data{}

    (PowerUpDataArray:powerup_player_data).SetWidgets(NewCanvas:canvas, NewStackBox:stack_box):void=
        set PowerUpDataArray.Canvas = NewCanvas
        set PowerUpDataArray.StackBox = NewStackBox
        set PowerUpDataArray.bIsCanvasSet = true

    (PowerUpDataArray:powerup_player_data).SetPowerupWidgetsRef(PowerupID:powerup_id, NewTexture:texture_block, NewTextBlock:text_block, NewPowerupStackBox:stack_box):void=
        #if data was not initialize before this initializes it
        if(not PowerUpDataArray.Data[PowerupID]):
            PowerUpDataArray.GetData(PowerupID)
        
        if(PowerUpUpgradeData:=PowerUpDataArray.Data[PowerupID]):
            PowerUpUpgradeData.SetWidgets(NewTexture, NewTextBlock, NewPowerupStackBox)

    (PowerUpDataArray:powerup_player_data).SetTask(NewTask:cancelable):void=      
        set PowerUpDataArray.MaybeTask=option{NewTask}

    (PowerUpDataArray:powerup_player_data).ContainsPowerUpData(PowerupID:powerup_id)<decides><transacts>:void=
        PowerUpDataArray.Data[PowerupID]  

    (PowerUpDataArray:powerup_player_data).SetPowerupUpgradeText(PowerupID:powerup_id, NewPowerupUpgrade:string):void=
         if(PowerUpUpgradeData:=PowerUpDataArray.Data[PowerupID]):
            PowerUpUpgradeData.SetUpgradeText(NewPowerupUpgrade)       

#=======================================================================================================================
#Enums

    powerup_id:= enum:
        None,
        MaxHealth,
        MaxShield,
        MaxDamage,
        ShieldRegen,
        HealthRegen

    powerup_upgrade_type:= enum:
        Add,
        Multiply

#=======================================================================================================================
#Tags

    powerup_device_tag:=class(tag){}
    player_spawner_device_tag:=class(tag){}

#=======================================================================================================================
#Utils
    # StringToMessage<localizes><public>(String:string)<computes>:message = "{String}"
    # AgentToMessage<localizes><public>(Agent:agent)<computes>:message = "{Agent}"

    # # CAREFUL: This function cannot be used with float values bigger than int
    # TruncateNumber(Number:float, Decimals:int)<transacts>:float =
    #     if:
    #         Decimals >= 0
    #         Multiplier := Pow(10.0, Decimals * 1.0)
    #         RoundValue := Round[Number * Multiplier] * 1.0
    #     then:
    #         RoundValue / Multiplier
    #     else:
    #         Number

    # ToTruncatedString<public>(Number:float, Decimals:int):[]char =
    #     var Str:[]char = ToString(Number)
    #     if:
    #         DotIndex := Str.Find['.']
    #         StopIndex := if (Decimals > 0) then Min(DotIndex+Decimals+1, Str.Length) else DotIndex
    #         if (Truncated := Str.Slice[0, StopIndex]):
    #             set Str = Truncated
    #     Str
#=======================================================================================================================
#Modules
    PowerUpSystem<public>:=module:
        Textures<public> := module:
            Icons<public> :=module:
                Legacy<public> :=module{}
            
#=========================================================================================================================

#Debug
    log_powerup_manager:=class(log_channel){}