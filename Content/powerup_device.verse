
using { /Fortnite.com/Devices }
using { /Verse.org/Simulation }
using { /Verse.org/Assets }
using { /UnrealEngine.com/Temporary/Diagnostics }

powerup_base_device := class(creative_device):

    Logger:log=log{Channel:=log_powerup_manager, DefaultLevel:=log_level.Debug}

    #When player increases this powerup how it happends, Add an increase to the current value, multiply a base value, etc...
    @editable
    UpgradeType:powerup_upgrade_type=powerup_upgrade_type.Add

    @editable
    BaseValue:float=1.0

    @editable
    BaseIncrementValue:float=1.0

    @editable
    MaxUpgrade:int=10

    @editable
    UpgradeCost:float=10.0

    @editable
    bPersistOnElimination:logic=true #If false, powerups are removed when player is eliminated

    @editable
    DecimalsToDisplay:int=0

    PowerupID:powerup_id=powerup_id.None
    PowerupDisplayName:string="Powerup Name"
    PowerupIcon:texture=PowerUpSystem.Textures.Icons.T_BlankPowerUp

    #Events

    OnPowerupStartedEvent:event(tuple(player, powerup_id))=event(tuple(player, powerup_id)){}
    #player, powerup_id, @int = UpgradeLevel
    OnPowerupUpgradedEvent:event(tuple(player, powerup_id, int))=event(tuple(player, powerup_id, int)){}
    OnPowerupFinishedEvent:event(tuple(player, powerup_id))=event(tuple(player, powerup_id)){}

    OnPlayerSpawnedEvent:event(player)=event(player){}
#=========================================================================================================================================
    OnBegin<override>()<suspends>:void={}
#=========================================================================================================================================
    AddPowerUpToPlayer(Player:player, UpgradeLevel:int):void=
        StartPowerupOnPlayer(Player, UpgradeLevel)
        OnPowerupStartedEvent.Signal(Player, PowerupID)
#=========================================================================================================================================
    RemovePowerupFromPlayer(Player:player):void=
        StopPowerupOnPlayer(Player)
        OnPowerupFinishedEvent.Signal(Player, PowerupID)
#=========================================================================================================================================
    StartPowerupOnPlayer<protected>(Player:player, UpgradeLevel:int):void=
        #Here you can set any requirement to start this powerup if required
        UpgradePowerup(Player, UpgradeLevel) 
        spawn{PowerupEventHandler(Player, UpgradeLevel)}
#=========================================================================================================================================
    PowerupLoopBehavior<protected>(Player:player, UpgradeLevel:int, TerminateHandlerEvent:event())<suspends>:void=
        TerminateLoopBehaviorEvent:=event(){}
        var CurrentUpgradeLevel:int=UpgradeLevel
        race:
            TerminateLoopBehaviorEvent.Await()
            sync:
                loop:
                    TerminateHandlerEvent.Await()
                    TerminateLoopBehaviorEvent.Signal()
                loop:
                    Sleep(1.0)
                    Logger.Print("Running: {PowerupDisplayName}")
                    TerminateLoopBehaviorEvent.Signal()
                    break               
#=========================================================================================================================================
    PowerupEventHandler<protected>(Player:player, UpgradeLevel:int)<suspends>:void=
        TerminateHandlerEvent:=event(){} 

        var CurrentUpgradeLevel:int=UpgradeLevel

        #Code inside this corutine
        spawn{PowerupLoopBehavior(Player, UpgradeLevel, TerminateHandlerEvent)}

        race:
            TerminateHandlerEvent.Await()
            sync:
                loop:
                    EventData:=OnPowerupFinishedEvent.Await()
                    EventPlayer:=EventData(0)
                    if(Player = EventPlayer):
                        TerminateHandlerEvent.Signal()
                loop:
                    EventData:=OnPowerupUpgradedEvent.Await()
                    EventPlayer:=EventData(0)
                    EventUpgradeLevel:=EventData(2)
                    if(Player = EventPlayer):
                        set CurrentUpgradeLevel = EventUpgradeLevel
                        OnPowerupUpgraded(Player, CurrentUpgradeLevel)
                loop:
                    SpawnedPlayer:=OnPlayerSpawnedEvent.Await()
                    Print("Player Spawned!")
                    if(Player=SpawnedPlayer):
                        Sleep(0.5)            
                        UpgradePowerup(Player, CurrentUpgradeLevel)
#=========================================================================================================================================
    StopPowerupOnPlayer<protected>(Player:player):void=
        Print("{PowerupDisplayName} Stopped!")
#=========================================================================================================================================
    RefreshPowerupUpgradeLevel(Player:player, UpgradeLevel:int):void= #Experimental Method not tested yet!
        StopPowerupOnPlayer(Player)
        StartPowerupOnPlayer(Player, UpgradeLevel)
#=========================================================================================================================================
<#Use this method to increase the player stat with the given Base value set#>
    CalculatePowerupValueByLevel(UpgradeLevel:int):float=
        var UpgradedValue:float=0.0

        case (UpgradeType):
            powerup_upgrade_type.Add =>
                set UpgradedValue = BaseValue + (BaseIncrementValue * UpgradeLevel)
            powerup_upgrade_type.Multiply =>
                set UpgradedValue = BaseValue * (1.0+(BaseIncrementValue * UpgradeLevel))

        return UpgradedValue
#=========================================================================================================================================
    #use this method to format the value you want to show on Store
    GetUIPowerupValueByLevel(UpgradeLevel:int):string=
        Value:=CalculatePowerupValueByLevel(UpgradeLevel)
        Truncated:=ToTruncatedString(Value, DecimalsToDisplay)
        return Truncated
#=========================================================================================================================================
    UpgradePowerup(Player:player, UpgradeLevel:int):void=
        <#Use this method to upgrade one call values like, max health, shield, etc...
            #>
        # ValueToUpgrade:=100.0 #E.g: Player Health, Shield, etc.
        # UpgradedValue:=CalculatePowerupValueByLevel(UpgradeLevel, ValueToUpgrade)
        #E.g: Set Player Max Health using @UpgradedValue 
        Logger.Print("{PowerupDisplayName}: Upgrading to {UpgradeLevel}")
        OnPowerupUpgradedEvent.Signal(Player, PowerupID, UpgradeLevel)
#=========================================================================================================================================
    OnPowerupUpgraded<protected>(Player:player, UpgradeLevel:int):void={} 
#=========================================================================================================================================
    GetPowerupMaxUpgrade():int=MaxUpgrade
#=========================================================================================================================================
    OnPlayerSpawned(Player:player):void=
        OnPlayerSpawnedEvent.Signal(Player)
