using { /Fortnite.com/Devices }
using { /Verse.org/Simulation }
using { /Verse.org/Assets }
using { /UnrealEngine.com/Temporary/Diagnostics }
using { /Fortnite.com/Characters }

powerup_health_regen_device := class(powerup_base_device):

    PowerupID<override>:powerup_id=powerup_id.HealthRegen
    PowerupDisplayName<override>:string="Health Regen"
    PowerupIcon<override>:texture=PowerUpSystem.Textures.Icons.Legacy.T_HealthRegen
    BaseValue<override>:float=0.0
    BaseIncrementValue<override>:float=0.5

    @editable
    RegenDelay:float=2.0
    
#=========================================================================================================================================
    PowerupLoopBehavior<override>(Player:player, UpgradeLevel:int, TerminateHandlerEvent:event())<suspends>:void=
        TerminateLoopBehaviorEvent:=event(){}
        var UpgradedHealthRegen:float=CalculatePowerupValueByLevel(UpgradeLevel)

        if(FortCharacter:=Player.GetFortCharacter[]):
            race:
                TerminateLoopBehaviorEvent.Await()
                sync:
                    loop:
                        TerminateHandlerEvent.Await()
                        TerminateLoopBehaviorEvent.Signal()
                    loop:
                        Sleep(RegenDelay)
                        if(FortCharacter.GetHealth() < FortCharacter.GetMaxHealth()):
                            FortCharacter.Heal(UpgradedHealthRegen)
                    loop:
                        EventData:=OnPowerupUpgradedEvent.Await()
                        EventPlayer:=EventData(0)
                        EventUpgradeLevel:=EventData(2)
                        if(Player = EventPlayer):
                            set UpgradedHealthRegen =CalculatePowerupValueByLevel(EventUpgradeLevel)
                    loop:
                        SpawnedPlayer:=OnPlayerSpawnedEvent.Await()
                        if(Player=SpawnedPlayer):
                            Print("Player Spawned!")
                            #Logic here :)