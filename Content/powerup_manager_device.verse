
using { /Fortnite.com/Devices }
using { /Fortnite.com/Characters }
using { /Fortnite.com/Game }
using { /Fortnite.com/UI }
using { /Verse.org/Simulation }
using { /Verse.org/Simulation/Tags }
using { /Verse.org/Colors }
using { /UnrealEngine.com/Temporary/Diagnostics }
using { /UnrealEngine.com/Temporary/UI }
using { /UnrealEngine.com/Temporary/SpatialMath }

powerup_manager_device := class(creative_device):

    Logger:log=log{Channel:=log_powerup_manager, DefaultLevel:=log_level.Debug}

    @editable
    PowerupTag:tag=powerup_device_tag{}
    @editable
    SpawnerTag:tag=player_spawner_device_tag{}
    @editable
    PowerupIconSize:vector2=vector2{X:=35.0,Y:=35.0}
    @editable
    PowerUpStackBoxSize:vector2=vector2{X:=650.0,Y:=70.0}
    @editable
    PowerUpStackBoxOffSet:vector2=vector2{X:=60.0,Y:=-120.0}

    var AvailablePowerUps:[]powerup_base_device=array{}

    #Store the status of the player's PowerUps for a session/round
    var PlayerPowerUpData:[player]powerup_player_data=map{}
#=========================================================================================================================
    OnBegin<override>()<suspends>:void=
        Sleep(1.0)
        set AvailablePowerUps = GetAllPowerUpDevices()
        GetPlayspace().PlayerRemovedEvent().Subscribe(OnPlayerRemovedEvent)

       GetAndBindAllSpawnerDevices()
#=========================================================================================================================
    GetAllPowerUpDevices():[]powerup_base_device=
        var PowerupDevices:[]powerup_base_device=array{}
        CreativeObjects:=FindCreativeObjectsWithTag(PowerupTag)
        for(CreativeObject:CreativeObjects, MaybePowerUpDevice:=powerup_base_device[CreativeObject]):
            set PowerupDevices+=array{MaybePowerUpDevice}
        Logger.Print("Powerups Found: {PowerupDevices.Length}")
        return PowerupDevices
#=========================================================================================================================
    GetAndBindAllSpawnerDevices():void=
        var SpawnerDevices:[]player_spawner_device=array{}
        CreativeObjects:=FindCreativeObjectsWithTag(SpawnerTag)
        for(CreativeObject:CreativeObjects, MaybeSpawnerDevice:=player_spawner_device[CreativeObject]):
            set SpawnerDevices+=array{MaybeSpawnerDevice}
            MaybeSpawnerDevice.SpawnedEvent.Subscribe(OnPlayerSpawnedEvent)
        Logger.Print("Spawners Found: {SpawnerDevices.Length}")
#=========================================================================================================================
    GetPlayerPowerUpUpgradeData(Player:player):powerup_player_data=
        if(TargetData:=PlayerPowerUpData[Player]):
            return TargetData
        else:
            NewData:=powerup_player_data{}
            if(set PlayerPowerUpData[Player] = NewData){}
            return NewData
#=========================================================================================================================
    GivePowerUpToPlayer(Player:player, PowerupID:powerup_id, UpgradeLevel:int):void=
        #Get PowerUp Data
        TargetPlayerData:=GetPlayerPowerUpUpgradeData(Player)

        MaybePowerUp:=GetPowerUpByID(PowerupID)
        if(TargetPowerup:=MaybePowerUp?):
            <#If this booster is not set to player, create a new canvas if need and Add it,
             or just add the new powerup to the existing canvas#> 
        
             #Base Powerup Multiplier
             PowerupValue:=TargetPowerup.BaseIncrementValue * UpgradeLevel
             #Only for Widget
             Decimals:=TargetPowerup.DecimalsToDisplay
             TruncatedString:=TargetPowerup.GetUIPowerupValueByLevel(UpgradeLevel)

            if(TargetPlayerData.Data[PowerupID].bIsPowerUpOwned?):
                Logger.Print("Upgrading PowerUp")
                TargetPowerup.UpgradePowerup(Player,UpgradeLevel)
                TargetPlayerData.SetPowerupUpgradeText(PowerupID, TruncatedString)
            else:
                Logger.Print("Creating New PowerUp")
                TargetPowerup.AddPowerUpToPlayer(Player, UpgradeLevel)
                #Create data to store powerup stuff
                TargetPlayerData.AddData(PowerupID, powerup_upgrade_data{})
                CreatePlayerUI(Player, TargetPowerup, TargetPlayerData)
                TargetPlayerData.SetPowerupUpgradeText(PowerupID, TruncatedString)
                #Subscribe event when player is eliminated
                if(not TargetPowerup.bPersistOnElimination?):
                    if(FortCharacter:= Player.GetFortCharacter[]):
                        Task:=FortCharacter.EliminatedEvent().Subscribe(OnPlayerEliminatedEvent)
                        TargetPlayerData.SetTask(Task) #Referenced tasks can be cancel
                   
#=========================================================================================================================
    CreatePlayerUI(Player:player, PowerUpDevice:powerup_base_device, TargetPlayerData:powerup_player_data):void=
    
        NewTextureBlock:= texture_block{DefaultImage:=PowerUpDevice.PowerupIcon, DefaultDesiredSize:=PowerupIconSize}
        UpgradeTextBlock:text_block=text_block{DefaultTextColor:=NamedColors.White}

        if(not TargetPlayerData.bIsCanvasSet?):
            Logger.Print("Creating New Canvas")
            WidgetData:=CreatePowerUpsWidget(Player)
            NewCanvas:=WidgetData(0)
            NewStackBox:=WidgetData(1)
            TargetPlayerData.SetWidgets(NewCanvas, NewStackBox)   
            Logger.Print("Updating Canvas")

        TargetCanvas:=TargetPlayerData.Canvas
        TargetStackBox:= TargetPlayerData.StackBox

        PowerupStackBox:=stack_box:
            Orientation:=orientation.Vertical
            Slots:=array:

                stack_box_slot:
                    Widget:=UpgradeTextBlock
                    Padding:=margin{}
                    HorizontalAlignment:=horizontal_alignment.Center
                    VerticalAlignment:=vertical_alignment.Center       
                    
                stack_box_slot:
                    Widget:=NewTextureBlock
                    Padding:=margin{}
                    HorizontalAlignment:=horizontal_alignment.Center
                    VerticalAlignment:=vertical_alignment.Center

                stack_box_slot:
                    Widget:=text_block{DefaultText:=StringToMessage("   ")} #Space to Keep Size
                    Padding:=margin{}
                    HorizontalAlignment:=horizontal_alignment.Center
                    VerticalAlignment:=vertical_alignment.Center 

        NewSlot:=stack_box_slot:
            Widget:=PowerupStackBox
            Padding:=margin{Top:=10.0,Right:=10.0,Left:=10.0,Bottom:=10.0}
            HorizontalAlignment:=horizontal_alignment.Center
            VerticalAlignment:=vertical_alignment.Center


        TargetPlayerData.SetPowerupWidgetsRef(PowerUpDevice.PowerupID, NewTextureBlock, UpgradeTextBlock, PowerupStackBox)
        TargetStackBox.AddWidget(NewSlot)

#=========================================================================================================================
    RemovePowerupFromPlayer(Player:player, PowerupID:powerup_id):void=
        if(TargetPlayerData:=PlayerPowerUpData[Player]):
            TargetStackBox:=TargetPlayerData.StackBox         
            if(TargetPowerupData:=TargetPlayerData.Data[PowerupID]):
                if(TargetPowerupStackBox:=TargetPowerupData.MaybePowerupStackBox?):
                    Logger.Print("Texture Block Valid")
                    TargetStackBox.RemoveWidget(TargetPowerupStackBox)
                    if(PlayerUI:=GetPlayerUI[Player]):
                        PlayerUI.RemoveWidget(TargetPowerupStackBox)
                else:
                    Logger.Print("Texture Block Invalid")
            else:
                Logger.Print("Invalid PowerUp Data")

            TargetPlayerData.ResetData(PowerupID)

        MaybePowerUpDevice:=GetPowerUpByID(PowerupID)
        if(TargetPowerupDevice:=MaybePowerUpDevice?):
            TargetPowerupDevice.RemovePowerupFromPlayer(Player)
#=========================================================================================================================
    RemovePlayerFromMap(Player:player):[player]powerup_player_data=
        TargetMap:=PlayerPowerUpData
        var NewMap:[player]powerup_player_data = map{}
        # Concatenate Keys from TargetMap into NewMap, excluding Player
        for (Key -> Value : TargetMap, Key <> Player):
            set NewMap = ConcatenateMaps(NewMap, map{Key => Value})
        return NewMap
#=========================================================================================================================
    CreatePowerUpsWidget(Player:player, ?TargetCanvas:?canvas=false):tuple(canvas, stack_box)=      
 
        #if TargetCanvas = false, create a new canvas

        #Canvas to store the PowerUp's icon
        var PowerUpCanvas:canvas=canvas{}

        if(ValidCanvas:=TargetCanvas?):
            set PowerUpCanvas = ValidCanvas

       #Main Overlay
       #Define Main Widget
       MainOverlay:=overlay{}
       PowerUpsStackBox:=stack_box{Orientation:=orientation.Horizontal}

       #Add Memebers
       Slot:=overlay_slot{
        Widget:=PowerUpsStackBox 
        HorizontalAlignment:=horizontal_alignment.Fill
        VerticalAlignment:=vertical_alignment.Fill
        Padding:=margin{Left:=5.0,Top:=5.0,Right:=5.0,Bottom:=5.0}
        }

       MainOverlay.AddWidget(Slot)

       PowerUpCanvas.AddWidget(CanvasSlot:=canvas_slot{
        Widget:=MainOverlay
        Anchors:=anchors{Minimum:=vector2{X:=0.0,Y:=1.0}, Maximum:=vector2{X:=0.0,Y:=1.0}}
        Offsets:=margin{Left:=PowerUpStackBoxOffSet.X, Top:=PowerUpStackBoxOffSet.Y, Right:=PowerUpStackBoxSize.X, Bottom:=PowerUpStackBoxSize.Y}
        Alignment:=vector2{X:=0.0, Y:=1.0}
       })

        if(PlayerUI:=GetPlayerUI[Player]):
            PlayerUI.AddWidget(PowerUpCanvas)

        return(PowerUpCanvas, PowerUpsStackBox)
#=========================================================================================================================================
    OnPlayerSpawnedEvent(SpawnedAgent:agent):void=
        if(Player:=player[SpawnedAgent]):
            for(TargetPowerup:AvailablePowerUps):
                TargetPowerup.OnPlayerSpawned(Player)
#=========================================================================================================================================
    OnPlayerEliminatedEvent(Result:elimination_result):void=
        EliminatedCharacter:=Result.EliminatedCharacter
        if(Player:=player[EliminatedCharacter.GetAgent[]]):
            ResetPlayerPowerUps(Player)
#=========================================================================================================================================
    OnPlayerRemovedEvent(Player:player):void=
        ResetPlayerPowerUps(Player)
        set PlayerPowerUpData = RemovePlayerFromMap(Player)
#=========================================================================================================================================
    ResetPlayerPowerUps(Player:player):void=
        for(PowerUp:AvailablePowerUps):
            Logger.Print("Removing: {PowerUp.PowerupDisplayName}")
            RemovePowerupFromPlayer(Player,PowerUp.PowerupID)
#=========================================================================================================================================
    GetPowerUpByID(PowerupID:powerup_id):?powerup_base_device=
        for(TargetPowerup:AvailablePowerUps):
            if(TargetPowerup.PowerupID = PowerupID):
                return option{TargetPowerup}
        return false
#=========================================================================================================================================
    GetPowerupMaxUpgrade(PowerupID:powerup_id):int=
        MaybePowerup:=GetPowerUpByID(PowerupID)
        if(ValidPowerup:=MaybePowerup?):
            return ValidPowerup.MaxUpgrade
        return 0