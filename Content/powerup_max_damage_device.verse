using { /Fortnite.com/Devices }
using { /Verse.org/Simulation }
using { /Verse.org/Assets }
using { /UnrealEngine.com/Temporary/Diagnostics }
using { /Fortnite.com/Characters }

powerup_max_damage_device := class(powerup_base_device):

    UpgradeType<override>:powerup_upgrade_type=powerup_upgrade_type.Multiply
    PowerupID<override>:powerup_id=powerup_id.MaxDamage
    PowerupDisplayName<override>:string="Damage Boost"
    PowerupIcon<override>:texture=PowerUpSystem.Textures.Icons.Legacy.T_DamageAmplifier
    BaseIncrementValue<override>:float=0.1
    DecimalsToDisplay<override>:int=1
    @editable
    DamageAmplifierDevice:damage_amplifier_powerup_device=damage_amplifier_powerup_device{}
#=========================================================================================================================================
    StopPowerupOnPlayer<override>(Player:player):void=
        BaseMaxDamage:=BaseValue
        Print("Restoring, Max Damage set to: {BaseMaxDamage}")
        SetPlayerMaxDamage(Player, BaseMaxDamage)
        (super:)StopPowerupOnPlayer(Player)
#=========================================================================================================================================
    UpgradePowerup<override>(Player:player, UpgradeLevel:int):void=
        UpgradedMaxDamage:=CalculatePowerupValueByLevel(UpgradeLevel)
        SetPlayerMaxDamage(Player, UpgradedMaxDamage)
        (super:)UpgradePowerup(Player, UpgradeLevel)
#=========================================================================================================================================
    SetPlayerMaxDamage(Player:player, NewMaxDamage:float):void=
        Print("Max Damage: {NewMaxDamage}")
        DamageAmplifierDevice.SetMagnitude(NewMaxDamage)
        DamageAmplifierDevice.Spawn()
        DamageAmplifierDevice.Pickup(Player)
