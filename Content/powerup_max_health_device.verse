using { /Fortnite.com/Devices }
using { /Verse.org/Simulation }
using { /Verse.org/Assets }
using { /UnrealEngine.com/Temporary/Diagnostics }
using { /Fortnite.com/Characters }

powerup_max_health_device := class(powerup_base_device):

    PowerupID<override>:powerup_id=powerup_id.MaxHealth
    PowerupDisplayName<override>:string="Max Health"
    PowerupIcon<override>:texture=PowerUpSystem.Textures.Icons.Legacy.T_MaxHealth
    BaseValue<override>:float=100.0
    BaseIncrementValue<override>:float=10.0
#=========================================================================================================================================
    StopPowerupOnPlayer<override>(Player:player):void=  
        MaxHealth:=BaseValue
        Print("Restoring, Max Health set to: {MaxHealth}")
        SetPlayerMaxHealth(Player, MaxHealth)
        (super:)StopPowerupOnPlayer(Player)
#=========================================================================================================================================
    UpgradePowerup<override>(Player:player, UpgradeLevel:int):void=
        # MaxHealth:=FortCharacter.GetMaxHealth()
        MaxHealth:=BaseValue
        UpgradedMaxHealth:=CalculatePowerupValueByLevel(UpgradeLevel)
        SetPlayerMaxHealth(Player,UpgradedMaxHealth)
        (super:)UpgradePowerup(Player, UpgradeLevel)
#=========================================================================================================================================
    SetPlayerMaxHealth(Player:player, NewMaxHealth:float):void=
        if(FortCharacter:=Player.GetFortCharacter[]):
            FortCharacter.SetMaxHealth(NewMaxHealth)