using { /Fortnite.com/Devices }
using { /Verse.org/Simulation }
using { /Verse.org/Assets }
using { /UnrealEngine.com/Temporary/Diagnostics }
using { /Fortnite.com/Characters }

powerup_max_shield_device := class(powerup_base_device):

    PowerupID<override>:powerup_id=powerup_id.MaxShield
    PowerupDisplayName<override>:string="Max Shield"
    PowerupIcon<override>:texture=PowerUpSystem.Textures.Icons.Legacy.T_MaxShield
    BaseValue<override>:float=100.0
    BaseIncrementValue<override>:float=10.0
#=========================================================================================================================================
    StopPowerupOnPlayer<override>(Player:player):void=
        MaxShield:=BaseValue
        Print("Restoring, Max Shield set to: {MaxShield}")
        SetPlayerMaxShield(Player, BaseValue)
        (super:)StopPowerupOnPlayer(Player)
#=========================================================================================================================================
    UpgradePowerup<override>(Player:player, UpgradeLevel:int):void=
        # MaxShield:=FortCharacter.GetMaxShield()
        MaxShield:=BaseValue
        UpgradedMaxShield:=CalculatePowerupValueByLevel(UpgradeLevel)
        SetPlayerMaxShield(Player, UpgradedMaxShield)
        (super:)UpgradePowerup(Player, UpgradeLevel)
#=========================================================================================================================================
    SetPlayerMaxShield(Player:player, NewMaxShield:float):void=
        if(FortCharacter:=Player.GetFortCharacter[]):
            FortCharacter.SetMaxShield(NewMaxShield)