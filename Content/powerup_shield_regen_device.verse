using { /Fortnite.com/Devices }
using { /Verse.org/Simulation }
using { /Verse.org/Assets }
using { /UnrealEngine.com/Temporary/Diagnostics }
using { /Fortnite.com/Characters }

powerup_shield_regen_device := class(powerup_base_device):

    PowerupID<override>:powerup_id=powerup_id.ShieldRegen
    PowerupDisplayName<override>:string="Shield Regen"
    PowerupIcon<override>:texture=PowerUpSystem.Textures.Icons.Legacy.T_ShieldRegen
    BaseValue<override>:float=0.0
    BaseIncrementValue<override>:float=0.5
    @editable
    RegenDelay:float=2.0

#=========================================================================================================================================
    PowerupLoopBehavior<override>(Player:player, UpgradeLevel:int, TerminateHandlerEvent:event())<suspends>:void=
        TerminateLoopBehaviorEvent:=event(){}
        var UpgradedShieldRegen:float=CalculatePowerupValueByLevel(UpgradeLevel)

        if(FortCharacter:=Player.GetFortCharacter[]):
            race:
                TerminateLoopBehaviorEvent.Await()
                sync:
                    loop:
                        TerminateHandlerEvent.Await()
                        TerminateLoopBehaviorEvent.Signal()
                    loop:
                        Sleep(RegenDelay)
                        if(CurrentShield:=FortCharacter.GetShield() < FortCharacter.GetMaxShield()):              
                            FortCharacter.SetShield(CurrentShield+UpgradedShieldRegen)
                    loop:
                        EventData:=OnPowerupUpgradedEvent.Await()
                        EventPlayer:=EventData(0)
                        EventUpgradeLevel:=EventData(2)
                        if(Player = EventPlayer):
                            Logger.Print("Upgrading Shield to Level: {EventUpgradeLevel}")
                            set UpgradedShieldRegen=CalculatePowerupValueByLevel(EventUpgradeLevel)
