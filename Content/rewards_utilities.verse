using { /Verse.org/Simulation }
using { /Verse.org/Random }

# 
HandleRarity(Rarity:item_rarity):item_rarity=  
    var RarityIndex:int = RarityToIndex(Rarity) 
    MaxIndex := item_rarity_id_array.Length - 1
    MinIndex := 0
    if (RarityIndex >= MaxIndex - 2 and RarityIndex <= MaxIndex):
        set RarityIndex -= 1
        NewRarity :=  IndexToRarity(RarityIndex) 
        NewIndex := RarityToIndex(NewRarity)       
        return IndexToRarity(RarityIndex) 
    else if(RarityIndex >= MinIndex and RarityIndex <= + 2):
        set RarityIndex += 1  
        NewRarity :=  IndexToRarity(RarityIndex) 
        NewIndex := RarityToIndex(NewRarity)     
        return IndexToRarity(RarityIndex) 
    return item_rarity.Common    

# Test. It will be replaced by player persistence.  
rarity_weight_config := struct<concrete>:
    @editable
    Rarity: item_rarity = item_rarity.Common
    @editable
    Weight: int = -1
        
# Test. It will be replaced by player persistence.  
rarity_distribution := class<concrete>:
    @editable
    RarityWeights:[]rarity_weight_config = array{}

    SelectRandomRarity(): item_rarity =
        var TotalWeight:int= 0
        for (RarityWeight:RarityWeights):
            set TotalWeight += RarityWeight.Weight
        RandomValue := GetRandomInt(0,TotalWeight)
        var AccumulatedWeight:int= 0
        for (RarityWeight:RarityWeights):
            set AccumulatedWeight += RarityWeight.Weight 
            if (RandomValue <= AccumulatedWeight):
                return RarityWeight.Rarity
        return item_rarity.Common 
        
# Test. It will be replaced by player persistence.        
# GetRarityByDistribution<private>(): item_rarity =
#     Number := GetRandomInt(0, 100)        
#     var NewRarity:item_rarity = item_rarity.Common
#     for (Data:RarityDistributionData):
#         if:
#             Number >= Data.MinValue and Number <= Data.MaxValue
#         then:
#             return Data.Rarity        
#     return NewRarity