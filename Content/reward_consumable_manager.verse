using { /Verse.org/Random }
using { /Verse.org/Assets }
using { /Verse.org/Simulation }
using { /Fortnite.com/Devices }
using { /UnrealEngine.com/Temporary/Diagnostics }

log_reward_consumable_manager := class(log_channel){}

reward_consumable_manager := class(creative_device,awardable):

    Logger:log = log{Channel:=log_reward_consumable_manager}

    @editable
    ConsumableDefinition:reward_consumable_definition = reward_consumable_definition{}  

    GetRewardableData<override>(RewardableData:rewardable_data): rewardable =
        var NewRarity:item_rarity = RewardableData.Rarity
        var MaxIterations: int = 5
        var CurrentIteration: int = 0        
        loop:
            bHasRarity := HasRarity(NewRarity)
            if:
                bHasRarity?
            then:
                break             
            if:                
                CurrentIteration >= MaxIterations 
            then:                
                set NewRarity = item_rarity.Common 
                break    
            set NewRarity = HandleRarity(NewRarity) 
            set CurrentIteration += 1
        return GetRandomRewardableByRarity(rewardable_data{Team:=RewardableData.Team, TypeSearch:=RewardableData.TypeSearch,Rarity:=NewRarity}) 

    GetRandomRewardableByRarity(RewardableData:rewardable_data): rewardable =
        var NewItem:item_data = item_data{} 
        if:
            RewardableData.TypeSearch = type_search.All
        then:
            set NewItem = FindItemByRarity(GetItemsData(),RewardableData.Rarity)
        else:
            AllItemsByTeam := FindItemsByTeam(GetItemsData(), RewardableData.Team)
            bHasTeam :=  HasRarity(AllItemsByTeam,RewardableData.Rarity)
            if:
                bHasTeam?
                AllItemsByTeam.Length > 0
            then:
                set NewItem = FindItemByRarity(AllItemsByTeam,RewardableData.Rarity)
            else:
                set NewItem = FindItemByRarity(GetItemsData(),RewardableData.Rarity) 
        if (ConsumableData := consumable_data[NewItem]) then return ConsumableData
        return item_data{}         

    GetItemsData():[]item_data=
        ConsumableDefinition.GetConsumableData()

    FindItemByRarity(ItemsData:[]item_data, Rarity:item_rarity):item_data =
        Logger.Print("Conusamble Manager")
        var ItemsByRarity:[]item_data = array{}
        var NewItem:item_data = item_data{}   
        for (ItemData:ItemsData):
            if (ItemData.Rarity = Rarity):
                set ItemsByRarity += array{ItemData} 
        if:
            ItemsByRarity.Length > 0            
        then:
            CurrenItem := GetRandomItem(ItemsByRarity)
            set NewItem = CurrenItem     
        return NewItem  
        
    FindItemsByTeam(ItemsData:[]item_data, Team:team_id):[]item_data =
        var ItemsByTeam:[]item_data = array{}
        for (ItemData:ItemsData):
            if (ItemData.Team = Team):
                set ItemsByTeam += array{ItemData} 
        if:
            ItemsByTeam.Length <= 0            
        then:
            Logger.Print("There are not elements") 
        return ItemsByTeam  

    GetRandomItem<private>(ItemsData:[]item_data): item_data=
        Length := ItemsData.Length
        var NewItem: item_data = item_data{}
        IndexItemRandon := GetRandomInt(0, ItemsData.Length-1)
        if:
            ItemChoosen := ItemsData[IndexItemRandon]
            ConsumableData := consumable_data[ItemChoosen]
        then:
            set NewItem = ConsumableData       
        return NewItem    

    GiveReward<override>(Player:player, Rewardable:rewardable): void=
        if:
            Item := item_data[Rewardable]  
            ItemGranterIndex := Item.ItemGranterIndex
            Agent := agent[Player]
        then:
            GrantItem(Agent, ItemGranterIndex)

    GrantItem(Agent:agent, ItemGranterIndex:int): void =
        ConsumableDefinition.GrantItemToPlayer(Agent, ItemGranterIndex)   

    HasRarity(Rarity:item_rarity): logic=
        ItemsData := GetItemsData()
        for (ItemData:ItemsData):
            if (ItemData.Rarity = Rarity):
                return true
        return false

    HasRarity(ItemsData:[]item_data,Rarity:item_rarity): logic=        
        for (ItemData:ItemsData):
            if (ItemData.Rarity = Rarity):
                return true
        return false

consumable_data := class<concrete>(item_data, iconTextureable, attributable):
    @editable
    ItemId: item_consumable_id = item_consumable_id.Undefined 

    GetRarity<override>()<computes>: item_rarity = Rarity

    GetDisplayName<override>()<computes>: string = DisplayName

    GetIconTexture<override>()<computes>: texture = FindConsumableIconTexture(ItemId)

    GetMaybeAttributes<override>()<computes>:?[]item_attribute_data = option{Attributes}

reward_consumable_definition := class<concrete>:

    @editable 
    ConsumableGranter<private>:item_granter_device = item_granter_device{}	

    GrantItemToPlayer<public>(Agent:agent, Index:int):void= 		
        ConsumableGranter.GrantItemIndex(Agent, Index)

    GetConsumableData<public>():[]consumable_data =
        return ConsumableData

    @editable
    ConsumableData<private>:[]consumable_data := array{
        consumable_data
        {
            ItemId:= item_consumable_id.Bandage_L1,
            DisplayName:= "Bandage",
            Rarity := item_rarity.Common,
            ItemGranterIndex := 0,
            Team := team_id.Blue,
            Attributes := array{         
                item_attribute_data{Id := item_attribute.Health, Value:= 15.0}}
        },
        consumable_data
        {
            ItemId:= item_consumable_id.ChugJug_L5,
            DisplayName:= "Chug Jug",
            Rarity := item_rarity.Legendary,
            ItemGranterIndex := 1,
            Team := team_id.Blue,
            Attributes := array{         
                item_attribute_data{Id := item_attribute.Health, Value:= 200.0}}
        },
        consumable_data
        {
            ItemId:= item_consumable_id.Flopper_L2,
            DisplayName:= "Flopper",
            Rarity := item_rarity.Uncommon,
            ItemGranterIndex := 2,
            Team := team_id.Blue,
            Attributes := array{         
                item_attribute_data{Id := item_attribute.Health, Value:= 40.0}}
            
        },
        consumable_data
        {
            ItemId:= item_consumable_id.Medkit_L2,
            DisplayName:= "Medkit",
            Rarity := item_rarity.Uncommon,
            ItemGranterIndex := 3,
            Team := team_id.Blue,
            Attributes := array{         
                item_attribute_data{Id := item_attribute.Health, Value:= 100.0}}
            
        },
        consumable_data
        {
            ItemId:= item_consumable_id.Mushroom_L1,
            DisplayName:= "Mushroom",
            Rarity := item_rarity.Common,
            ItemGranterIndex := 4,
            Team := team_id.Blue,
            Attributes := array{         
                item_attribute_data{Id := item_attribute.Shield, Value:= 5.0}}
            
        },
        consumable_data
        {
            ItemId:= item_consumable_id.ShieldPotion_L3,
            DisplayName:= "Shield Potion",
            Rarity := item_rarity.Rare,
            ItemGranterIndex := 5,
            Team := team_id.Blue,
            Attributes := array{         
                item_attribute_data{Id := item_attribute.Shield, Value:= 50.0}}
            
        },
        consumable_data
        {
            ItemId:= item_consumable_id.Shieldfish_L3,
            DisplayName:= "Shieldfish",
            Rarity := item_rarity.Rare,
            ItemGranterIndex := 6,
            Team := team_id.Blue,
            Attributes := array{         
                item_attribute_data{Id := item_attribute.Shield, Value:= 50.0}}
            
        },
        consumable_data
        {
            ItemId:= item_consumable_id.SlurpJuice_L4,
            DisplayName:= "Slurp Juice",
            Rarity := item_rarity.Epic,
            ItemGranterIndex := 7,
            Team := team_id.Blue,
            Attributes := array{         
                item_attribute_data{Id := item_attribute.Health, Value:= 75.0}}
            
        },
        consumable_data
        {
            ItemId:= item_consumable_id.SlurpMushroom_L1,
            DisplayName:= "Slurp Mushroom",
            Rarity := item_rarity.Common,
            ItemGranterIndex := 8,
            Team := team_id.Blue,
            Attributes := array{         
                item_attribute_data{Id := item_attribute.Health, Value:= 10.0}}
            
        },
        consumable_data
        {
            ItemId:= item_consumable_id.Slurpfish_L4,
            DisplayName:= "Slurpfish",
            Rarity := item_rarity.Epic,
            ItemGranterIndex := 9,
            Team := team_id.Blue,
            Attributes := array{         
                item_attribute_data{Id := item_attribute.Health, Value:= 40.0}}
            
        },
        consumable_data
        {
            ItemId:= item_consumable_id.SmallFry_L1,
            DisplayName:= "Small Fry",
            Rarity := item_rarity.Common,
            ItemGranterIndex := 10,
            Team := team_id.Blue,
            Attributes := array{         
                item_attribute_data{Id := item_attribute.Health, Value:= 25.0}}
            
        },
        consumable_data
        {
            ItemId:= item_consumable_id.SmallShieldPotion_L2,
            DisplayName:= "Small Shield Potion",
            Rarity := item_rarity.Uncommon,
            ItemGranterIndex := 11,
            Team := team_id.Blue,
            Attributes := array{         
                item_attribute_data{Id := item_attribute.Shield, Value:= 25.0}}
            
        },
        consumable_data
        {
            ItemId:= item_consumable_id.SpicyFish_L3,
            DisplayName:= "Spicy Fish",
            Rarity := item_rarity.Rare,
            ItemGranterIndex := 12,
            Team := team_id.Blue,
            Attributes := array{         
                item_attribute_data{Id := item_attribute.Health, Value:= 15.0}}            
        }              
    }