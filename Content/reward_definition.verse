using { /Fortnite.com/Devices }
using { /Verse.org/Simulation }
using { /Verse.org/Assets }
using { /UnrealEngine.com/Temporary/Diagnostics }

<# 
Interfaces for anything that can be rewarded.
#>
rewardable := interface():
    GetRarity():item_rarity
    GetDisplayName()<computes>:string

describable := interface():
    GetDescription()<transacts>:string

iconTextureable := interface():
    GetIconTexture()<computes>:texture

attributable := interface():
    GetMaybeAttributes()<computes>:?[]item_attribute_data

<# 
Auxiliary classes.
#>
rewardable_data := class:
    Team:team_id
    TypeSearch:type_search
    Rarity:item_rarity

reward_selection_payload := struct:
    Player:player
    Reward:rewardable
    RewardType:reward_type

room_reward_data := class:    
    Rewardables:[]rewardable
    RewardClaimed:logic

<# 
Interface to grant rewards.
#>
awardable := interface():
    GiveReward(Player:player, Rewardable:rewardable): void
    GetRewardableData(RewardableData:rewardable_data): rewardable

reward_definition := class(creative_device):

    @editable
    WeaponManager:reward_weapon_manager = reward_weapon_manager{}

    @editable
    ConsumableManager:reward_consumable_manager = reward_consumable_manager{}

    <#
    @editable
    CurrencyManager:reward_currency_manager = reward_currency_manager{}

    @editable
    XPManager:reward_xp_manager = reward_xp_manager{} #>

    @editable
    BoosterManager:clash_booster_manager_device = clash_booster_manager_device{}   

    GetAwardableReward(Id:reward_type): awardable=
        Print("")
        var MaybeAwardable:?awardable = false
        case (Id):  
            reward_type.Weapon  => set MaybeAwardable = option{WeaponManager}
            reward_type.Consumable => set MaybeAwardable = option{ConsumableManager}
            #reward_type.Currency => set MaybeAwardable = option{CurrencyManager}
            #reward_type.XP => set MaybeAwardable = option{XPManager}  
            reward_type.Booster => set MaybeAwardable = option{BoosterManager}            
        if:
            Awardable := MaybeAwardable?
        then:
            return Awardable
        return reward_weapon_manager{}

reward_type := enum {Weapon, Consumable<#, Currency, XP#>, Booster}

RewardTypeToMessage(Id:reward_type) : string =
    case(Id):
        reward_type.Weapon => "Weapon"
        reward_type.Consumable => "Consumable"
        #reward_type.Currency => "Currency"
        #reward_type.XP => "XP"  
        reward_type.Booster => "Booster"   
        # ...