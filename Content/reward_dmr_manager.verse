using { /Fortnite.com/Devices }
using { /Verse.org/Random }
using { /Verse.org/Assets }
using { /Verse.org/Simulation }
using { /UnrealEngine.com/Temporary/Diagnostics }

log_reward_dmr_manager := class(log_channel){}

reward_dmr_manager := class(creative_device,grantable):

    Logger:log = log{Channel:=log_reward_dmr_manager}

    @editable
    DMRDefinition<private>:reward_dmr_definition = reward_dmr_definition{}        

    FindItemByRarity<override>(ItemsData:[]item_data, Rarity:item_rarity):item_data =
        Logger.Print("DMR Manager")
        var ItemsByRarity:[]item_data = array{}
        var ItemData:item_data = item_data{}  
        for (TempItemData:ItemsData):
            if (TempItemData.Rarity = Rarity):
                set ItemsByRarity += array{TempItemData} 
        if:
            ItemsByRarity.Length > 0            
        then:
            set ItemData = GetRandomItem(ItemsByRarity)       
        return ItemData 

    FindItemsByTeam<override>(ItemsData:[]item_data, Team:team_id):[]item_data =
        var ItemsByTeam:[]item_data = array{}
        for (ItemData:ItemsData):
            if (ItemData.Team = Team):
                set ItemsByTeam += array{ItemData} 
        if:
            ItemsByTeam.Length <= 0            
        then:
            Logger.Print("There are not elements") 
        return ItemsByTeam  

    GetRandomItem<private>(ItemsData:[]item_data): item_data=
        var NewItem: item_data = item_data{}
        IndexItemRandon := GetRandomInt(0, ItemsData.Length-1)
        if:
            ItemChoosen := ItemsData[IndexItemRandon]
            DMRData := dmr_data[ItemChoosen]
        then:
            set NewItem = DMRData
        return NewItem	 

    GrantItem<override>(Agent:agent, Index:int):void= 
        Logger.Print("DMR Granter")
        DMRDefinition.GrantItemToPlayer(Agent, Index)

    GetItemsData<override>():[]item_data =
        DMRDefinition.GetDMRData() 

    HasRarity<override>(Rarity:item_rarity): logic=
        ItemsData := GetItemsData()
        for (ItemData:ItemsData):
            if (ItemData.Rarity = Rarity):
                return true
        return false

    HasRarity<override>(ItemsData:[]item_data,Rarity:item_rarity): logic=        
        for (ItemData:ItemsData):
            if (ItemData.Rarity = Rarity):
                return true
        return false

dmr_data := class<concrete>(weapon_data, iconTextureable):
    @editable
    ItemId: item_dmr_id = item_dmr_id.Undefined 
    
    WeaponType<override>:weapon = weapon.Dmr
    
    GetIconTexture<override>()<computes>: texture = FindDMRIconTexture(ItemId) 

reward_dmr_definition := class<concrete>:
    @editable # One item granter with all the DMRs
    DMRGranter<private>:item_granter_device = item_granter_device{}	

    GrantItemToPlayer<public>(Agent:agent, Index:int):void= 	
        DMRGranter.GrantItemIndex(Agent, Index)
        
    GetDMRData<public>():[]item_data =
        return DMRData

    @editable
    DMRData<private>:[]dmr_data := array{	
        dmr_data
        {
            ItemId:= item_dmr_id.CobraDMR_L1,
            DisplayName:= "Cobra DMR",
            Rarity := item_rarity.Common,
            ItemGranterIndex := 0,
            Team := team_id.Blue,   
            Attributes := array{         
                item_attribute_data{Id := item_attribute.DPS, Value:= 144.0},
                item_attribute_data{Id := item_attribute.Damage, Value:= 36.0},              
                item_attribute_data{Id := item_attribute.MagSize, Value:= 20.0}
            }
        },
        dmr_data
        {
            ItemId:= item_dmr_id.CobraDMR_L2,
            DisplayName:= "Cobra DMR",
            Rarity := item_rarity.Uncommon,
            ItemGranterIndex := 1,
            Team := team_id.Blue,   
            Attributes := array{         
                item_attribute_data{Id := item_attribute.DPS, Value:= 152.0},
                item_attribute_data{Id := item_attribute.Damage, Value:= 38.0},              
                item_attribute_data{Id := item_attribute.MagSize, Value:= 20.0}
            }
        },
        dmr_data
        {
            ItemId:= item_dmr_id.CobraDMR_L3,
            DisplayName:= "Cobra DMR",
            Rarity := item_rarity.Rare,
            ItemGranterIndex := 2,
            Team := team_id.Blue,   
            Attributes := array{         
                item_attribute_data{Id := item_attribute.DPS, Value:= 160.0},
                item_attribute_data{Id := item_attribute.Damage, Value:= 40.0},              
                item_attribute_data{Id := item_attribute.MagSize, Value:= 20.0}
            }
        },
        dmr_data
        {
            ItemId:= item_dmr_id.CobraDMR_L4,
            DisplayName:= "Cobra DMR",
            Rarity := item_rarity.Epic,
            ItemGranterIndex := 3,
            Team := team_id.Blue,   
            Attributes := array{         
                item_attribute_data{Id := item_attribute.DPS, Value:= 168.0},
                item_attribute_data{Id := item_attribute.Damage, Value:= 42.0},              
                item_attribute_data{Id := item_attribute.MagSize, Value:= 20.0}
            }
        },
        dmr_data
        {
            ItemId:= item_dmr_id.CobraDMR_L5,
            DisplayName:= "Cobra DMR",
            Rarity := item_rarity.Legendary,
            ItemGranterIndex := 4,
            Team := team_id.Blue,   
            Attributes := array{         
                item_attribute_data{Id := item_attribute.DPS, Value:= 176.0},
                item_attribute_data{Id := item_attribute.Damage, Value:= 44.0},              
                item_attribute_data{Id := item_attribute.MagSize, Value:= 20.0}
            }
        },
        dmr_data
        {
            ItemId:= item_dmr_id.CobraDMR_L6,
            DisplayName:= "Cobra DMR",
            Rarity := item_rarity.Mythic,
            ItemGranterIndex := 5,
            Team := team_id.Blue,   
            Attributes := array{         
                item_attribute_data{Id := item_attribute.DPS, Value:= 184.0},
                item_attribute_data{Id := item_attribute.Damage, Value:= 46.0},              
                item_attribute_data{Id := item_attribute.MagSize, Value:=20.0}
            }
        },
        dmr_data
        {
            ItemId:= item_dmr_id.ThermalDMR_L1,
            DisplayName:= "Thermal DMR",
            Rarity := item_rarity.Common,
            ItemGranterIndex := 6,
            Team := team_id.Red,   
            Attributes := array{         
                item_attribute_data{Id := item_attribute.DPS, Value:= 160.0},
                item_attribute_data{Id := item_attribute.Damage, Value:= 32.0},              
                item_attribute_data{Id := item_attribute.MagSize, Value:= 20.0}
            }
        },
        dmr_data
        {
            ItemId:= item_dmr_id.ThermalDMR_L2,
            DisplayName:= "Thermal DMR",
            Rarity := item_rarity.Uncommon,
            ItemGranterIndex := 7,
            Team := team_id.Red,  
            Attributes := array{         
                item_attribute_data{Id := item_attribute.DPS, Value:= 165.0},
                item_attribute_data{Id := item_attribute.Damage, Value:= 33.0},              
                item_attribute_data{Id := item_attribute.MagSize, Value:= 20.0}
            }
        },
        dmr_data
        {
            ItemId:= item_dmr_id.ThermalDMR_L3,
            DisplayName:= "Thermal DMR",
            Rarity := item_rarity.Rare,
            ItemGranterIndex := 8,
            Team := team_id.Red,  
            Attributes := array{         
                item_attribute_data{Id := item_attribute.DPS, Value:= 175.0},
                item_attribute_data{Id := item_attribute.Damage, Value:= 35.0},              
                item_attribute_data{Id := item_attribute.MagSize, Value:= 20.0}
            }
        },
        dmr_data
        {
            ItemId:= item_dmr_id.ThermalDMR_L4,
            DisplayName:= "Thermal DMR",
            Rarity := item_rarity.Epic,
            ItemGranterIndex := 9,
            Team := team_id.Red,  
            Attributes := array{         
                item_attribute_data{Id := item_attribute.DPS, Value:= 185.0},
                item_attribute_data{Id := item_attribute.Damage, Value:= 37.0},              
                item_attribute_data{Id := item_attribute.MagSize, Value:= 20.0}
            }
        },
        dmr_data
        {
            ItemId:= item_dmr_id.ThermalDMR_L5,
            DisplayName:= "Thermal DMR",
            Rarity := item_rarity.Legendary,
            ItemGranterIndex := 10,
            Team := team_id.Red,  
            Attributes := array{         
                item_attribute_data{Id := item_attribute.DPS, Value:= 195.0},
                item_attribute_data{Id := item_attribute.Damage, Value:= 39.0},              
                item_attribute_data{Id := item_attribute.MagSize, Value:= 20.0}
            }
        },
        dmr_data
        {
            ItemId:= item_dmr_id.ThermalDMR_L6,
            DisplayName:= "Thermal DMR",
            Rarity := item_rarity.Mythic,
            ItemGranterIndex := 11,
            Team := team_id.Red,  
            Attributes := array{         
                item_attribute_data{Id := item_attribute.DPS, Value:= 200.0},
                item_attribute_data{Id := item_attribute.Damage, Value:= 40.0},              
                item_attribute_data{Id := item_attribute.MagSize, Value:= 20.0}
            }
        }
    }