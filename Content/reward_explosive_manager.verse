using { /Fortnite.com/Devices }
using { /Verse.org/Random }
using { /Verse.org/Assets }
using { /Verse.org/Simulation }
using { /UnrealEngine.com/Temporary/Diagnostics }

log_reward_explosive_manager := class(log_channel){}

reward_explosive_manager := class(creative_device,grantable):

    Logger:log = log{Channel:=log_reward_explosive_manager}

    @editable
    ExplosiveDefinition<private>:reward_explosive_definition = reward_explosive_definition{}   

    FindItemByRarity<override>(ItemsData:[]item_data, Rarity:item_rarity):item_data =
        Logger.Print("Explosive Manager")
        var ItemsByRarity:[]item_data = array{}
        var ItemData:item_data = item_data{}  
        for (TempItemData:ItemsData):
            if (TempItemData.Rarity = Rarity):
                set ItemsByRarity += array{TempItemData} 
        if:
            ItemsByRarity.Length > 0            
        then:
            set ItemData = GetRandomItem(ItemsByRarity)       
        return ItemData 

    FindItemsByTeam<override>(ItemsData:[]item_data, Team:team_id):[]item_data =
        var ItemsByTeam:[]item_data = array{}
        for (ItemData:ItemsData):
            if (ItemData.Team = Team):
                set ItemsByTeam += array{ItemData} 
        if:
            ItemsByTeam.Length <= 0            
        then:
            Logger.Print("There are not elements") 
        return ItemsByTeam  

    GetRandomItem<private>(ItemsData:[]item_data): item_data=
        var NewItem: item_data = item_data{}
        IndexItemRandon := GetRandomInt(0, ItemsData.Length-1)
        if:
            ItemChoosen := ItemsData[IndexItemRandon]
            ExplosiveData := explosive_data[ItemChoosen]
        then:
            set NewItem = ExplosiveData
        return NewItem	    

    GrantItem<override>(Agent:agent, Index:int):void= 
        Logger.Print("Explosive Granter")
        ExplosiveDefinition.GrantItemToPlayer(Agent, Index)

    GetItemsData<override>():[]item_data =
        ExplosiveDefinition.GetExplosiveData()   

    HasRarity<override>(Rarity:item_rarity): logic=
        ItemsData := GetItemsData()
        for (ItemData:ItemsData):
            if (ItemData.Rarity = Rarity):
                return true
        return false

    HasRarity<override>(ItemsData:[]item_data,Rarity:item_rarity): logic=        
        for (ItemData:ItemsData):
            if (ItemData.Rarity = Rarity):
                return true
        return false
        
explosive_data := class<concrete>(weapon_data, iconTextureable):
    @editable
    ItemId: item_explosive_id = item_explosive_id.Undefined  
    
    WeaponType<override>:weapon = weapon.Explosive
    
    GetIconTexture<override>()<computes>: texture = FindExplosiveIconTexture(ItemId)

reward_explosive_definition := class<concrete>:
    @editable # One item granter with all the Explosives
    ExplosiveGranter<private>:item_granter_device = item_granter_device{}	

    GrantItemToPlayer<public>(Agent:agent, Index:int):void= 		
        ExplosiveGranter.GrantItemIndex(Agent, Index)
        
    GetExplosiveData<public>():[]item_data =
        return ExplosiveData

    @editable
    ExplosiveData<private>:[]explosive_data := array{	
        explosive_data
        {
            ItemId:= item_explosive_id.BoomBolt_L3,
            DisplayName:= "Boom Bolt",
            Rarity := item_rarity.Rare,
            ItemGranterIndex := 0,
            Team := team_id.Blue,
            Attributes := array{         
                item_attribute_data{Id := item_attribute.DPS, Value:= 71.4}, item_attribute_data{Id := item_attribute.Damage, Value:= 102.0}, item_attribute_data{Id := item_attribute.MagSize, Value:= 3.0}
            }
        },
        explosive_data
        {
            ItemId:= item_explosive_id.BoomBolt_L4,
            DisplayName:= "Boom Bolt",
            Rarity := item_rarity.Epic,
            ItemGranterIndex := 1,
            Team := team_id.Blue,
            Attributes := array{         
                item_attribute_data{Id := item_attribute.DPS, Value:= 0.0}, item_attribute_data{Id := item_attribute.Damage, Value:= 0.0}, item_attribute_data{Id := item_attribute.MagSize, Value:= 0.0}
            }
        },
        explosive_data
        {
            ItemId:= item_explosive_id.BoomBolt_L5,
            DisplayName:= "Boom Bolt",
            Rarity := item_rarity.Legendary,
            ItemGranterIndex := 2,
            Team := team_id.Blue,
            Attributes := array{         
                item_attribute_data{Id := item_attribute.DPS, Value:= 0.0}, item_attribute_data{Id := item_attribute.Damage, Value:= 0.0}, item_attribute_data{Id := item_attribute.MagSize, Value:= 0.0}
            }
        },
        explosive_data
        {
            ItemId:= item_explosive_id.BoomBolt_L6,
            DisplayName:= "Boom Bolt",
            Rarity := item_rarity.Mythic,
            ItemGranterIndex := 3,
            Team := team_id.Blue,
            Attributes := array{         
                item_attribute_data{Id := item_attribute.DPS, Value:= 0.0}, item_attribute_data{Id := item_attribute.Damage, Value:= 0.0}, item_attribute_data{Id := item_attribute.MagSize, Value:= 0.0}
            }
        },
        explosive_data
        {
            ItemId:= item_explosive_id.BoomBow_L5,
            DisplayName:= "Boom Bow",
            Rarity := item_rarity.Legendary,
            ItemGranterIndex := 4,
            Team := team_id.Blue,
            Attributes := array{         
                item_attribute_data{Id := item_attribute.DPS, Value:= 0.0}, item_attribute_data{Id := item_attribute.Damage, Value:= 0.0}, item_attribute_data{Id := item_attribute.MagSize, Value:= 0.0}
            }
        },
        explosive_data
        {
            ItemId:= item_explosive_id.ExplosiveRepeaterRifle_L3,
            DisplayName:= "Explosive Repeater Rifle",
            Rarity := item_rarity.Rare,
            ItemGranterIndex := 5,
            Team := team_id.Blue,
            Attributes := array{         
                item_attribute_data{Id := item_attribute.DPS, Value:= 34.2}, item_attribute_data{Id := item_attribute.Damage, Value:= 36.0}, item_attribute_data{Id := item_attribute.MagSize, Value:= 5.0}
            }
        },
        explosive_data
        {
            ItemId:= item_explosive_id.ExplosiveRepeaterRifle_L4,
            DisplayName:= "Explosive Repeater Rifle",
            Rarity := item_rarity.Epic,
            ItemGranterIndex := 6,
            Team := team_id.Blue,
            Attributes := array{         
                item_attribute_data{Id := item_attribute.DPS, Value:= 0.0}, item_attribute_data{Id := item_attribute.Damage, Value:= 0.0}, item_attribute_data{Id := item_attribute.MagSize, Value:= 0.0}
            }
        },
        explosive_data
        {
            ItemId:= item_explosive_id.ExplosiveRepeaterRifle_L5,
            DisplayName:= "Explosive Repeater Rifle",
            Rarity := item_rarity.Legendary,
            ItemGranterIndex := 7,
            Team := team_id.Blue,
            Attributes := array{         
                item_attribute_data{Id := item_attribute.DPS, Value:= 0.0}, item_attribute_data{Id := item_attribute.Damage, Value:= 0.0}, item_attribute_data{Id := item_attribute.MagSize, Value:= 0.0}
            }
        },
        explosive_data
        {
            ItemId:= item_explosive_id.MechanicalExplosiveBow_L3,
            DisplayName:= "Mechanical Explosive Bow",
            Rarity := item_rarity.Rare,
            ItemGranterIndex := 8,
            Team := team_id.Blue,
            Attributes := array{         
                item_attribute_data{Id := item_attribute.DPS, Value:= 75.0}, item_attribute_data{Id := item_attribute.Damage, Value:= 75.0}, item_attribute_data{Id := item_attribute.MagSize, Value:= 1.0}
            }
        },
        explosive_data
        {
            ItemId:= item_explosive_id.MechanicalExplosiveBow_L4,
            DisplayName:= "Mechanical Explosive Bow",
            Rarity := item_rarity.Epic,
            ItemGranterIndex := 9,
            Team := team_id.Blue,
            Attributes := array{         
                item_attribute_data{Id := item_attribute.DPS, Value:= 0.0}, 
                item_attribute_data{Id := item_attribute.Damage, Value:= 0.0}, 
                item_attribute_data{Id := item_attribute.MagSize, Value:= 0.0}
            }
        },
        explosive_data
        {
            ItemId:= item_explosive_id.MechanicalExplosiveBow_L5,
            DisplayName:= "Mechanical Explosive Bow",
            Rarity := item_rarity.Legendary,
            ItemGranterIndex := 10,
            Team := team_id.Blue,
            Attributes := array{         
                item_attribute_data{Id := item_attribute.DPS, Value:= 0.0}, 
                item_attribute_data{Id := item_attribute.Damage, Value:= 0.0}, 
                item_attribute_data{Id := item_attribute.MagSize, Value:= 0.0}
            }
        },

        explosive_data
        {
            ItemId:= item_explosive_id.StickyGrenadeLauncher_L4,
            DisplayName:= "Sticky Grenade Launcher",
            Rarity := item_rarity.Epic,
            ItemGranterIndex := 11,
            Team := team_id.Blue,
            Attributes := array{         
                item_attribute_data{Id := item_attribute.DPS, Value:= 0.0}, 
                item_attribute_data{Id := item_attribute.Damage, Value:= 0.0}, 
                item_attribute_data{Id := item_attribute.MagSize, Value:= 0.0}
            }
        },
        explosive_data
        {
            ItemId:= item_explosive_id.StickyGrenadeLauncher_L5,
            DisplayName:= "Sticky Grenade Launcher",
            Rarity := item_rarity.Legendary,
            ItemGranterIndex := 12,
            Team := team_id.Blue,
            Attributes := array{         
                item_attribute_data{Id := item_attribute.DPS, Value:= 0.0}, 
                item_attribute_data{Id := item_attribute.Damage, Value:= 0.0}, 
                item_attribute_data{Id := item_attribute.MagSize, Value:= 0.0}
            }
        }   
    }