using { /Fortnite.com/Devices }
using { /Verse.org/Random }
using { /Verse.org/Simulation }
using { /Fortnite.com/Characters }
using { /UnrealEngine.com/Temporary/Diagnostics }

log_reward_manager := class(log_channel){}

<# 
Context Class that implements the strategy design pattern.
#>
reward_handler_context := class:
    var MaybeStrategy:?awardable = false 

    SetStrategy(NewStrategy:awardable):void=
        set MaybeStrategy = option{NewStrategy}

    GiveReward(Player:player, Rewardable:rewardable): void=
        if:
            Strategy := MaybeStrategy?
        then:
            Strategy.GiveReward(Player, Rewardable) 

    GetRewardableData(RewardableData:rewardable_data): rewardable =
        var ItemData:item_data = item_data{}
        if:
            Strategy := MaybeStrategy?
        then:
            Rewardable := Strategy.GetRewardableData(RewardableData)
            return Rewardable
        return ItemData      

<# 
Context Class that implements the strategy design pattern.
#>
reward_manager := class(creative_device): 
    
    Logger:log = log{Channel:=log_reward_manager}    

    @editable
    RarityDistribution:rarity_distribution = rarity_distribution{}

    @editable
    RewardDefinition:reward_definition = reward_definition{}

    @editable
	OpenSelectRewardInputDevice:input_trigger_device = input_trigger_device{}

    @editable
    UISelectReward:ui_select_reward = ui_select_reward{}   

    @editable
	UIThumbnailReward:ui_thumbnail_reward = ui_thumbnail_reward{}     

    OnBegin<override>()<suspends>:void=    
        OpenSelectRewardInputDevice.PressedEvent.Subscribe(OnOpenSelectRewardInputPressed)
        loop:            
            RewardSelection:reward_selection_payload = UISelectReward.PlayerSelectionEvent.Await()              
            GiveReward(RewardSelection)        
            #Vulnerability 
            SetPlayerVulnerability(RewardSelection.Player, true)

    GiveReward<public>(RewardSelectionPayload:reward_selection_payload): void=
        Player := RewardSelectionPayload.Player
        Reward := RewardSelectionPayload.Reward
        RewardType := RewardSelectionPayload.RewardType
        ContextStrategy := reward_handler_context{}
        SetStrategy(ContextStrategy, RewardType)
        SelectReward(ContextStrategy, Player, Reward)    

    ChooseRandomRewards(RoomDataTeam:room_data_team): []ui_rewardable_data=
        var Rewardables:[]ui_rewardable_data = array{}
        for (RoomDataReward:RoomDataTeam.RoomDataRewards):
            NewRarity is RarityDistribution.SelectRandomRarity()
            Logger.Print("RARITY : {RarityToMessage(NewRarity)}")
            CurrentRewardable :=  GetRewardableData(RoomDataReward.RewardType, rewardable_data{Team := RoomDataTeam.Team, TypeSearch:= RoomDataReward.TypeSearch, Rarity := NewRarity})           
            set Rewardables += array{ui_rewardable_data{Rewardable:=CurrentRewardable, RewardType:=RoomDataReward.RewardType}}    
        return Rewardables

    GetRewardableData(Id:reward_type, RewardableData:rewardable_data): rewardable=
        ContextStrategy := reward_handler_context{}
        SetStrategy(ContextStrategy, Id)
        Rewardable := ChooseRewardableData(ContextStrategy, RewardableData)   

    ChooseRewardableData(ContextStrategy:reward_handler_context,RewardableData:rewardable_data): rewardable=
        return Rewardable := ContextStrategy.GetRewardableData(RewardableData) 

    SelectReward(ContextStrategy:reward_handler_context, Player:player, Reward:rewardable):void=
        ContextStrategy.GiveReward(Player,Reward) 

    ShowUIReward(Player:player,RewardableRoom:rewardableRoom,Index:int): void=          
        if:
            Index > 0
        then:       
            # If not lobby shows thumbnail.
            ShowThumbnailRoomReward(Player, RewardableRoom,Index)            
        else:
            # If it is lobby (first room) it shows the reward selection UI. 
            PlayerRewards := RewardableRoom.GetUIRewardableData(Player)                 
            UISelectReward.ShowUI(Player, PlayerRewards, RewardableRoom)   
            #Vulnerability
            SetPlayerVulnerability(Player, false)
        
    ShowThumbnailRoomReward(Player:player, Rewards:rewardableRoom,Index:int): void =
        PlayerRewards := Rewards.GetUIRewardableData(Player)
        if (PlayerRewards.Length > 0):
            UIThumbnailReward.Show(Player,Rewards)
            OpenSelectRewardInputDevice.Register(Player)       

    HideAnyPreviewRoomReward(Player:player): void =
        #Logger.Print("Hide Any Preview Room Reward")        
        if (UIThumbnailReward.GetUIPlayerData[Player]):
            UIThumbnailReward.Hide(Player)
            OpenSelectRewardInputDevice.Unregister(Player)

    # Event called when a player presses the request reward button. If there is a thumbnail, it shows the selection UI, otherwise it shows nothing.
    OnOpenSelectRewardInputPressed(Agent:agent): void =
        if:
            Player:=player[Agent]
            FortCharacter:=Agent.GetFortCharacter[]
            not FortCharacter.IsDownButNotOut[] # Do not allow to claim reward if down but not out
            UIPlayerData:=UIThumbnailReward.GetUIPlayerData[Player]
            CurrentUIPlayerData := UIThumbnailReward.GetUIPlayerData[Player]                       
        then:
            UIRewardableDatas := CurrentUIPlayerData.RewardableRooms.GetUIRewardableData(Player)
            Room := CurrentUIPlayerData.RewardableRooms
            HideAnyPreviewRoomReward(Player)
            UISelectReward.ShowUI(Player, UIRewardableDatas, Room)     
            #Vulnerability
            SetPlayerVulnerability(Player, false)      
        else:
            #Logger.Print("Fail to Hide prev")

    SetStrategy(ContextStrategy:reward_handler_context,Id:reward_type):void=
        case (Id):
            reward_type.Weapon => ContextStrategy.SetStrategy(RewardDefinition.GetAwardableReward(Id))
            reward_type.Consumable => ContextStrategy.SetStrategy(RewardDefinition.GetAwardableReward(Id))
            #reward_type.Currency => ContextStrategy.SetStrategy(RewardDefinition.GetAwardableReward(Id))
            #reward_type.XP => ContextStrategy.SetStrategy(RewardDefinition.GetAwardableReward(Id))    
            reward_type.Booster => ContextStrategy.SetStrategy(RewardDefinition.GetAwardableReward(Id))  

    SetPlayerVulnerability(Player:player, Vulnerable:logic):void=
        if(Agent:=agent[Player],FC:=Agent.GetFortCharacter[]):
            FC.SetVulnerability(Vulnerable)
