using { /Verse.org/Random }
using { /Verse.org/Assets }
using { /Verse.org/Simulation }
using { /Fortnite.com/Devices }
using { /UnrealEngine.com/Temporary/Diagnostics }

log_reward_others_manager := class(log_channel){}

reward_others_manager := class(creative_device,grantable):

    Logger:log = log{Channel:=log_reward_others_manager}

    @editable
    OthersDefinition:reward_others_definition = reward_others_definition{}

    FindItemByRarity<override>(ItemsData:[]item_data, Rarity:item_rarity):item_data =
        Logger.Print("Others Manager")
        var ItemsByRarity:[]item_data = array{}
        var ItemData:item_data = item_data{}  
        for (TempItemData:ItemsData):
            if (TempItemData.Rarity = Rarity):
                set ItemsByRarity += array{TempItemData} 
        if:
            ItemsByRarity.Length > 0            
        then:
            set ItemData = GetRandomItem(ItemsByRarity)       
        return ItemData 

    FindItemsByTeam<override>(ItemsData:[]item_data, Team:team_id):[]item_data =
        var ItemsByTeam:[]item_data = array{}
        for (ItemData:ItemsData):
            if (ItemData.Team = Team):
                set ItemsByTeam += array{ItemData} 
        if:
            ItemsByTeam.Length <= 0            
        then:
            Logger.Print("There are not elements") 
        return ItemsByTeam  

    GetRandomItem<private>(ItemsData:[]item_data): item_data=
        var NewItem: item_data = item_data{}
        IndexItemRandon := GetRandomInt(0, ItemsData.Length-1)
        if:
            ItemChoosen := ItemsData[IndexItemRandon]
            OthersData := others_data[ItemChoosen]
        then:
            set NewItem = OthersData
        return NewItem	 

    GrantItem<override>(Agent:agent, Index:int):void= 
        Logger.Print("Others Granter")
        OthersDefinition.GrantItemToPlayer(Agent, Index)

    GetItemsData<override>():[]item_data =
        OthersDefinition.GetOthersData() 

    HasRarity<override>(Rarity:item_rarity): logic=
        ItemsData := GetItemsData()
        for (ItemData:ItemsData):
            if (ItemData.Rarity = Rarity):
                return true
        return false

    HasRarity<override>(ItemsData:[]item_data,Rarity:item_rarity): logic=        
        for (ItemData:ItemsData):
            if (ItemData.Rarity = Rarity):
                return true
        return false

others_data := class<concrete>(weapon_data, iconTextureable):
    @editable
    ItemId: item_others_id = item_others_id.Undefined 
    
    WeaponType<override>:weapon = weapon.Others

    GetIconTexture<override>()<computes>: texture = FindOthersIconTexture(ItemId)

reward_others_definition := class<concrete>:
    @editable # 
    OthersGranter<private>:item_granter_device = item_granter_device{}	

    GrantItemToPlayer<public>(Agent:agent, Index:int):void= 		
        OthersGranter.GrantItemIndex(Agent, Index)

    GetOthersData<public>():[]others_data =
        return OthersData

    @editable
    OthersData<private>:[]others_data := array{
        others_data
        {
            ItemId:= item_others_id.BasicHammer_L1,
            DisplayName:= "Basic Hammer",
            Rarity := item_rarity.Common,
            ItemGranterIndex := 0,
            Team := team_id.Blue,
            Attributes := array{         
                item_attribute_data{Id := item_attribute.DPS, Value:= 0.0}, item_attribute_data{Id := item_attribute.Damage, Value:= 0.0}, item_attribute_data{Id := item_attribute.MagSize, Value:= 0.0}

            }
        },
        others_data
        {
            ItemId:= item_others_id.BasicSword_L1,
            DisplayName:= "Basic Sword",
            Rarity := item_rarity.Common,
            ItemGranterIndex := 1,
            Team := team_id.Blue,
            Attributes := array{         
                item_attribute_data{Id := item_attribute.DPS, Value:= 0.0}, item_attribute_data{Id := item_attribute.Damage, Value:= 0.0}, item_attribute_data{Id := item_attribute.MagSize, Value:= 0.0}

            }
        },
        others_data
        {
            ItemId:= item_others_id.Chainsaw_L4,
            DisplayName:= "Chainsaw",
            Rarity := item_rarity.Epic,
            ItemGranterIndex := 2,
            Team := team_id.Blue,
            Attributes := array{         
                item_attribute_data{Id := item_attribute.DPS, Value:= 0.0}, item_attribute_data{Id := item_attribute.Damage, Value:= 0.0}, item_attribute_data{Id := item_attribute.MagSize, Value:= 0.0}

            }
        },
        others_data
        {
            ItemId:= item_others_id.Crossbow_L3,
            DisplayName:= "Crossbow",
            Rarity := item_rarity.Rare,
            ItemGranterIndex := 3,
            Team := team_id.Blue,
            Attributes := array{         
                item_attribute_data{Id := item_attribute.DPS, Value:= 0.0}, item_attribute_data{Id := item_attribute.Damage, Value:= 0.0}, item_attribute_data{Id := item_attribute.MagSize, Value:= 0.0}

            }
        },
        others_data
        {
            ItemId:= item_others_id.Crossbow_L4,
            DisplayName:= "Crossbow",
            Rarity := item_rarity.Epic,
            ItemGranterIndex := 4,
            Team := team_id.Blue,
            Attributes := array{         
                item_attribute_data{Id := item_attribute.DPS, Value:= 0.0}, item_attribute_data{Id := item_attribute.Damage, Value:= 0.0}, item_attribute_data{Id := item_attribute.MagSize, Value:= 0.0}

            }
        },
        others_data
        {
            ItemId:= item_others_id.KineticBlade_L3,
            DisplayName:= "Kinetic Blade",
            Rarity := item_rarity.Rare,
            ItemGranterIndex := 5,
            Team := team_id.Blue,
            Attributes := array{         
                item_attribute_data{Id := item_attribute.DPS, Value:= 0.0}, item_attribute_data{Id := item_attribute.Damage, Value:= 0.0}, item_attribute_data{Id := item_attribute.MagSize, Value:= 0.0}

            }
        },
        others_data
        {
            ItemId:= item_others_id.KineticBlade_L4,
            DisplayName:= "Kinetic Blade",
            Rarity := item_rarity.Epic,
            ItemGranterIndex := 6,
            Team := team_id.Blue,
            Attributes := array{         
                item_attribute_data{Id := item_attribute.DPS, Value:= 0.0}, item_attribute_data{Id := item_attribute.Damage, Value:= 0.0}, item_attribute_data{Id := item_attribute.MagSize, Value:= 0.0}

            }
        },
        others_data
        {
            ItemId:= item_others_id.MakeshiftBow_L2,
            DisplayName:= "Makeshift Bow",
            Rarity := item_rarity.Uncommon,
            ItemGranterIndex := 7,
            Team := team_id.Blue,
            Attributes := array{         
                item_attribute_data{Id := item_attribute.DPS, Value:= 0.0}, item_attribute_data{Id := item_attribute.Damage, Value:= 0.0}, item_attribute_data{Id := item_attribute.MagSize, Value:= 0.0}

            }
        },
        others_data
        {
            ItemId:= item_others_id.MechanicalBow_L3,
            DisplayName:= "Mechanical Bow",
            Rarity := item_rarity.Rare,
            ItemGranterIndex := 8,
            Team := team_id.Blue,
            Attributes := array{         
                item_attribute_data{Id := item_attribute.DPS, Value:= 0.0}, item_attribute_data{Id := item_attribute.Damage, Value:= 0.0}, item_attribute_data{Id := item_attribute.MagSize, Value:= 0.0}

            }
        }		     
    }