using { /Fortnite.com/Devices }
using { /Verse.org/Random }
using { /Verse.org/Assets }
using { /Verse.org/Simulation }
using { /UnrealEngine.com/Temporary/Diagnostics }

log_reward_pistol_manager := class(log_channel){}

reward_pistol_manager := class(creative_device,grantable):
   
    Logger:log = log{Channel:=log_reward_pistol_manager}

    @editable
    PistolDefinition<private>:reward_pistol_definition = reward_pistol_definition{}     

    FindItemByRarity<override>(ItemsData:[]item_data, Rarity:item_rarity):item_data =
        Logger.Print("Pistol Manager")
        var ItemsByRarity:[]item_data = array{}
        var ItemData:item_data = item_data{}  
        for (TempItemData:ItemsData):
            if (TempItemData.Rarity = Rarity):
                set ItemsByRarity += array{TempItemData} 
        if:
            ItemsByRarity.Length > 0            
        then:
            set ItemData = GetRandomItem(ItemsByRarity)       
        return ItemData 

    FindItemsByTeam<override>(ItemsData:[]item_data, Team:team_id):[]item_data =
        var ItemsByTeam:[]item_data = array{}
        for (ItemData:ItemsData):
            if (ItemData.Team = Team):
                set ItemsByTeam += array{ItemData} 
        if:
            ItemsByTeam.Length <= 0            
        then:
            Logger.Print("There are not elements") 
        return ItemsByTeam  

    GetRandomItem<private>(ItemsData:[]item_data): item_data=
        var NewItem: item_data = item_data{}
        IndexItemRandon := GetRandomInt(0, ItemsData.Length-1)
        if:
            ItemChoosen := ItemsData[IndexItemRandon]
            PistolData := pistol_data[ItemChoosen]
        then:
            set NewItem = PistolData
        return NewItem

    GrantItem<override>(Agent:agent, Index:int):void= 
        Logger.Print("Pistol Granter")
        PistolDefinition.GrantItemToPlayer(Agent, Index)

    GetItemsData<override>():[]item_data =
        PistolDefinition.GetPistolData() 

    HasRarity<override>(Rarity:item_rarity): logic=
        ItemsData := GetItemsData()
        for (ItemData:ItemsData):
            if (ItemData.Rarity = Rarity):
                return true
        return false

    HasRarity<override>(ItemsData:[]item_data,Rarity:item_rarity): logic=        
        for (ItemData:ItemsData):
            if (ItemData.Rarity = Rarity):
                return true
        return false

pistol_data := class<concrete>(weapon_data, iconTextureable):
    @editable
    ItemId: item_pistol_id = item_pistol_id.Undefined   
    
    WeaponType<override>:weapon = weapon.Pistol
    
    GetIconTexture<override>()<computes>: texture = FindPistolIconTexture(ItemId)


reward_pistol_definition := class<concrete>:
    @editable # One item granter with all the Pistols
    PistolGranter<private>:item_granter_device = item_granter_device{}	

    GrantItemToPlayer<public>(Agent:agent, Index:int):void= 		
        PistolGranter.GrantItemIndex(Agent, Index)
    
    GetPistolData<public>():[]item_data =
        return PistolData

    @editable
    PistolData<private>:[]pistol_data := array{	
        pistol_data
        {
            ItemId:= item_pistol_id.DualPistols_L2,
            DisplayName:= "Dual Pistols",
            Rarity := item_rarity.Uncommon,
            ItemGranterIndex := 0,
            Team := team_id.Blue,
            Attributes := array{         
                item_attribute_data{Id := item_attribute.DPS, Value:= 0.0},
            item_attribute_data{Id := item_attribute.Damage, Value:= 0.0},
            item_attribute_data{Id := item_attribute.MagSize, Value:= 0.0}
            }
        },
        pistol_data
        {
            ItemId:= item_pistol_id.DualPistols_L3,
            DisplayName:= "Dual Pistols",
            Rarity := item_rarity.Rare,
            ItemGranterIndex := 1,
            Team := team_id.Blue,
            Attributes := array{         
                item_attribute_data{Id := item_attribute.DPS, Value:= 0.0},
            item_attribute_data{Id := item_attribute.Damage, Value:= 0.0},
            item_attribute_data{Id := item_attribute.MagSize, Value:= 0.0}
            }
        },
        pistol_data
        {
            ItemId:= item_pistol_id.DualPistols_L4,
            DisplayName:= "Dual Pistols",
            Rarity := item_rarity.Epic,
            ItemGranterIndex := 2,
            Team := team_id.Blue,
            Attributes := array{         
                item_attribute_data{Id := item_attribute.DPS, Value:= 158.4},
            item_attribute_data{Id := item_attribute.Damage, Value:= 40.0},
            item_attribute_data{Id := item_attribute.MagSize, Value:= 18.0}
            }
        },
        pistol_data
        {
            ItemId:= item_pistol_id.DualPistols_L5,
            DisplayName:= "Dual Pistols",
            Rarity := item_rarity.Legendary,
            ItemGranterIndex := 3,
            Team := team_id.Blue,
            Attributes := array{         
                item_attribute_data{Id := item_attribute.DPS, Value:= 0.0},
            item_attribute_data{Id := item_attribute.Damage, Value:= 0.0},
            item_attribute_data{Id := item_attribute.MagSize, Value:= 0.0}
            }
        },
        pistol_data
        {
            ItemId:= item_pistol_id.HandCannon_L3,
            DisplayName:= "Hand Cannon",
            Rarity := item_rarity.Rare,
            ItemGranterIndex := 4,
            Team := team_id.Blue,
            Attributes := array{         
                item_attribute_data{Id := item_attribute.DPS, Value:= 59.5},
            item_attribute_data{Id := item_attribute.Damage, Value:= 85.0},
            item_attribute_data{Id := item_attribute.MagSize, Value:= 7.0}
            }
        },
        pistol_data
        {
            ItemId:= item_pistol_id.HandCannon_L4,
            DisplayName:= "Hand Cannon",
            Rarity := item_rarity.Epic,
            ItemGranterIndex := 5,
            Team := team_id.Blue,
            Attributes := array{         
                item_attribute_data{Id := item_attribute.DPS, Value:= 0.0},
            item_attribute_data{Id := item_attribute.Damage, Value:= 0.0},
            item_attribute_data{Id := item_attribute.MagSize, Value:= 0.0}
            }
        },
        pistol_data
        {
            ItemId:= item_pistol_id.HandCannon_L5,
            DisplayName:= "Hand Cannon",
            Rarity := item_rarity.Legendary,
            ItemGranterIndex := 6,
            Team := team_id.Blue,
            Attributes := array{         
                item_attribute_data{Id := item_attribute.DPS, Value:= 0.0},
            item_attribute_data{Id := item_attribute.Damage, Value:= 0.0},
            item_attribute_data{Id := item_attribute.MagSize, Value:= 0.0}
            }
        },
        pistol_data
        {
            ItemId:= item_pistol_id.MonarchPistol_L1,
            DisplayName:= "Monarch Pistol",
            Rarity := item_rarity.Common,
            ItemGranterIndex := 7,
            Team := team_id.Blue,
            Attributes := array{         
                item_attribute_data{Id := item_attribute.DPS, Value:= 0.0},
            item_attribute_data{Id := item_attribute.Damage, Value:= 0.0},
            item_attribute_data{Id := item_attribute.MagSize, Value:= 0.0}
            }
        },
        pistol_data
        {
            ItemId:= item_pistol_id.MonarchPistol_L2,
            DisplayName:= "Monarch Pistol",
            Rarity := item_rarity.Uncommon,
            ItemGranterIndex := 8,
            Team := team_id.Blue,
            Attributes := array{         
                item_attribute_data{Id := item_attribute.DPS, Value:= 0.0},
            item_attribute_data{Id := item_attribute.Damage, Value:= 0.0},
            item_attribute_data{Id := item_attribute.MagSize, Value:= 0.0}
            }
        },
        pistol_data
        {
            ItemId:= item_pistol_id.MonarchPistol_L3,
            DisplayName:= "Monarch Pistol",
            Rarity := item_rarity.Rare,
            ItemGranterIndex := 9,
            Team := team_id.Blue,
            Attributes := array{         
                item_attribute_data{Id := item_attribute.DPS, Value:= 103.2},
            item_attribute_data{Id := item_attribute.Damage, Value:= 43.0},
            item_attribute_data{Id := item_attribute.MagSize, Value:= 7.0}
            }
        },
        pistol_data
        {
            ItemId:= item_pistol_id.MonarchPistol_L4,
            DisplayName:= "Monarch Pistol",
            Rarity := item_rarity.Epic,
            ItemGranterIndex := 10,
            Team := team_id.Blue,
            Attributes := array{         
                item_attribute_data{Id := item_attribute.DPS, Value:= 0.0},
            item_attribute_data{Id := item_attribute.Damage, Value:= 0.0},
            item_attribute_data{Id := item_attribute.MagSize, Value:= 0.0}
            }
        },
        pistol_data
        {
            ItemId:= item_pistol_id.MonarchPistol_L5,
            DisplayName:= "Monarch Pistol",
            Rarity := item_rarity.Legendary,
            ItemGranterIndex := 11,
            Team := team_id.Blue,
            Attributes := array{         
                item_attribute_data{Id := item_attribute.DPS, Value:= 0.0},
            item_attribute_data{Id := item_attribute.Damage, Value:= 0.0},
            item_attribute_data{Id := item_attribute.MagSize, Value:= 0.0}
            }
        },
        pistol_data
        {
            ItemId:= item_pistol_id.RangerPistol_L1,
            DisplayName:= "Ranger Pistol",
            Rarity := item_rarity.Common,
            ItemGranterIndex := 12,
            Team := team_id.Blue,
            Attributes := array{         
                item_attribute_data{Id := item_attribute.DPS, Value:= 0.0},
            item_attribute_data{Id := item_attribute.Damage, Value:= 0.0},
            item_attribute_data{Id := item_attribute.MagSize, Value:= 0.0}
            }
        },
        pistol_data
        {
            ItemId:= item_pistol_id.RangerPistol_L2,
            DisplayName:= "Ranger Pistol",
            Rarity := item_rarity.Uncommon,
            ItemGranterIndex := 13,
            Team := team_id.Blue,
            Attributes := array{         
                item_attribute_data{Id := item_attribute.DPS, Value:= 0.0},
            item_attribute_data{Id := item_attribute.Damage, Value:= 0.0},
            item_attribute_data{Id := item_attribute.MagSize, Value:= 0.0}
            }
        },
        pistol_data
        {
            ItemId:= item_pistol_id.RangerPistol_L3,
            DisplayName:= "Ranger Pistol",
            Rarity := item_rarity.Rare,
            ItemGranterIndex := 14,
            Team := team_id.Blue,
            Attributes := array{         
                item_attribute_data{Id := item_attribute.DPS, Value:= 178.2},
            item_attribute_data{Id := item_attribute.Damage, Value:= 27.0},
            item_attribute_data{Id := item_attribute.MagSize, Value:= 15.0}
            }
        },
        pistol_data
        {
            ItemId:= item_pistol_id.RangerPistol_L4,
            DisplayName:= "Ranger Pistol",
            Rarity := item_rarity.Epic,
            ItemGranterIndex := 15,
            Team := team_id.Blue,
            Attributes := array{         
                item_attribute_data{Id := item_attribute.DPS, Value:= 0.0},
            item_attribute_data{Id := item_attribute.Damage, Value:= 0.0},
            item_attribute_data{Id := item_attribute.MagSize, Value:= 0.0}
            }
        },
        pistol_data
        {
            ItemId:= item_pistol_id.RangerPistol_L5,
            DisplayName:= "Ranger Pistol",
            Rarity := item_rarity.Legendary,
            ItemGranterIndex := 16,
            Team := team_id.Blue,
            Attributes := array{         
                item_attribute_data{Id := item_attribute.DPS, Value:= 0.0},
            item_attribute_data{Id := item_attribute.Damage, Value:= 0.0},
            item_attribute_data{Id := item_attribute.MagSize, Value:= 0.0}
            }
        },
        pistol_data 
        {
            ItemId:= item_pistol_id.Revolver_L1,
            DisplayName:= "Revolver",
            Rarity := item_rarity.Common,
            ItemGranterIndex := 17,
            Team := team_id.Blue,
            Attributes := array{         
                item_attribute_data{Id := item_attribute.DPS, Value:= 0.0},
            item_attribute_data{Id := item_attribute.Damage, Value:= 0.0},
            item_attribute_data{Id := item_attribute.MagSize, Value:= 0.0}
            }
        },
        pistol_data
        {
            ItemId:= item_pistol_id.Revolver_L2,
            DisplayName:= "Revolver",
            Rarity := item_rarity.Uncommon,
            ItemGranterIndex := 18,
            Team := team_id.Blue,
            Attributes := array{         
                item_attribute_data{Id := item_attribute.DPS, Value:= 0.0},
            item_attribute_data{Id := item_attribute.Damage, Value:= 0.0},
            item_attribute_data{Id := item_attribute.MagSize, Value:= 0.0}
            }
        },
        pistol_data
        {
            ItemId:= item_pistol_id.Revolver_L3,
            DisplayName:= "Revolver",
            Rarity := item_rarity.Rare,
            ItemGranterIndex := 19,
            Team := team_id.Blue,
            Attributes := array{         
                item_attribute_data{Id := item_attribute.DPS, Value:= 90.0},
            item_attribute_data{Id := item_attribute.Damage, Value:= 60.0},
            item_attribute_data{Id := item_attribute.MagSize, Value:= 6.0}
            }
        },
        pistol_data
        {
            ItemId:= item_pistol_id.Revolver_L4,
            DisplayName:= "Revolver",
            Rarity := item_rarity.Epic,
            ItemGranterIndex := 20,
            Team := team_id.Blue,
            Attributes := array{         
                item_attribute_data{Id := item_attribute.DPS, Value:= 0.0},
            item_attribute_data{Id := item_attribute.Damage, Value:= 0.0},
            item_attribute_data{Id := item_attribute.MagSize, Value:= 0.0}
            }
        },
        pistol_data
        {
            ItemId:= item_pistol_id.Revolver_L5,
            DisplayName:= "Revolver",
            Rarity := item_rarity.Legendary,
            ItemGranterIndex := 21,
            Team := team_id.Blue,
            Attributes := array{         
                item_attribute_data{Id := item_attribute.DPS, Value:= 0.0},
            item_attribute_data{Id := item_attribute.Damage, Value:= 0.0},
            item_attribute_data{Id := item_attribute.MagSize, Value:= 0.0}
            }
        },
        pistol_data
        {
            ItemId:= item_pistol_id.SixShooter_L2,
            DisplayName:= "Six Shooter",
            Rarity := item_rarity.Uncommon,
            ItemGranterIndex := 22,
            Team := team_id.Blue,
            Attributes := array{         
                item_attribute_data{Id := item_attribute.DPS, Value:= 0.0},
            item_attribute_data{Id := item_attribute.Damage, Value:= 0.0},
            item_attribute_data{Id := item_attribute.MagSize, Value:= 0.0}
            }
        },
        pistol_data
        {
            ItemId:= item_pistol_id.SixShooter_L3,
            DisplayName:= "Six Shooter",
            Rarity := item_rarity.Rare,
            ItemGranterIndex := 23,
            Team := team_id.Blue,
            Attributes := array{         
                item_attribute_data{Id := item_attribute.DPS, Value:= 0.0},
            item_attribute_data{Id := item_attribute.Damage, Value:= 0.0},
            item_attribute_data{Id := item_attribute.MagSize, Value:= 0.0}
            }
        },
        pistol_data
        {
            ItemId:= item_pistol_id.SixShooter_L4,
            DisplayName:= "Six Shooter",
            Rarity := item_rarity.Epic,
            ItemGranterIndex := 24,
            Team := team_id.Blue,
            Attributes := array{         
                item_attribute_data{Id := item_attribute.DPS, Value:= 0.0},
            item_attribute_data{Id := item_attribute.Damage, Value:= 0.0},
            item_attribute_data{Id := item_attribute.MagSize, Value:= 0.0}
            }
        },
        pistol_data
        {
            ItemId:= item_pistol_id.TacticalPistol_L1,
            DisplayName:= "Tactical Pistol",
            Rarity := item_rarity.Common,
            ItemGranterIndex := 25,
            Team := team_id.Blue,
            Attributes := array{         
                item_attribute_data{Id := item_attribute.DPS, Value:= 0.0},
            item_attribute_data{Id := item_attribute.Damage, Value:= 0.0},
            item_attribute_data{Id := item_attribute.MagSize, Value:= 0.0}
            }
        },
        pistol_data
        {
            ItemId:= item_pistol_id.TacticalPistol_L2,
            DisplayName:= "Tactical Pistol",
            Rarity := item_rarity.Uncommon,
            ItemGranterIndex := 26,
            Team := team_id.Blue,
            Attributes := array{         
                item_attribute_data{Id := item_attribute.DPS, Value:= 0.0},
            item_attribute_data{Id := item_attribute.Damage, Value:= 0.0},
            item_attribute_data{Id := item_attribute.MagSize, Value:= 0.0}
            }
        },
        pistol_data
        {
            ItemId:= item_pistol_id.TacticalPistol_L3,
            DisplayName:= "Tactical Pistol",
            Rarity := item_rarity.Rare,
            ItemGranterIndex := 27,
            Team := team_id.Blue,
            Attributes := array{         
                item_attribute_data{Id := item_attribute.DPS, Value:= 170.0},
            item_attribute_data{Id := item_attribute.Damage, Value:= 25.0},
            item_attribute_data{Id := item_attribute.MagSize, Value:= 15.0}
            }
        },
        pistol_data
        {
            ItemId:= item_pistol_id.TacticalPistol_L4,
            DisplayName:= "Tactical Pistol",
            Rarity := item_rarity.Epic,
            ItemGranterIndex := 28,
            Team := team_id.Blue,
            Attributes := array{         
                item_attribute_data{Id := item_attribute.DPS, Value:= 0.0},
            item_attribute_data{Id := item_attribute.Damage, Value:= 0.0},
            item_attribute_data{Id := item_attribute.MagSize, Value:= 0.0}
            }
        },
        pistol_data
        {
            ItemId:= item_pistol_id.TacticalPistol_L5,
            DisplayName:= "Tactical Pistol",
            Rarity := item_rarity.Legendary,
            ItemGranterIndex := 29,
            Team := team_id.Blue,
            Attributes := array{         
                item_attribute_data{Id := item_attribute.DPS, Value:= 0.0},
            item_attribute_data{Id := item_attribute.Damage, Value:= 0.0},
            item_attribute_data{Id := item_attribute.MagSize, Value:= 0.0}
            }
        },
        pistol_data
        {
            ItemId:= item_pistol_id.TacticalPistol_L6,
            DisplayName:= "Tactical Pistol",
            Rarity := item_rarity.Mythic,
            ItemGranterIndex := 30,
            Team := team_id.Blue,
            Attributes := array{         
                item_attribute_data{Id := item_attribute.DPS, Value:= 0.0},
            item_attribute_data{Id := item_attribute.Damage, Value:= 0.0},
            item_attribute_data{Id := item_attribute.MagSize, Value:= 0.0}
            }
        }
    }