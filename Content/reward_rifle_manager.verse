using { /Fortnite.com/Devices }
using { /Verse.org/Random }
using { /Verse.org/Assets }
using { /Verse.org/Simulation }
using { /UnrealEngine.com/Temporary/Diagnostics }

log_reward_rifle_manager := class(log_channel){}

reward_rifle_manager := class(creative_device,grantable):

    Logger:log = log{Channel:=log_reward_rifle_manager}

    @editable
    RifleDefinition<private>:reward_rifle_definition = reward_rifle_definition{}   
    
    FindItemByRarity<override>(ItemsData:[]item_data, Rarity:item_rarity):item_data =
        Logger.Print("Rifle Manager")
        var ItemsByRarity:[]item_data = array{}
        var ItemData:item_data = item_data{}  
        for (TempItemData:ItemsData):
            if (TempItemData.Rarity = Rarity):
                set ItemsByRarity += array{TempItemData} 
        if:
            ItemsByRarity.Length > 0            
        then:
            set ItemData = GetRandomItem(ItemsByRarity)       
        return ItemData 

    FindItemsByTeam<override>(ItemsData:[]item_data, Team:team_id):[]item_data =
        var ItemsByTeam:[]item_data = array{}
        for (ItemData:ItemsData):
            if (ItemData.Team = Team):
                set ItemsByTeam += array{ItemData} 
        if:
            ItemsByTeam.Length <= 0            
        then:
            Logger.Print("There are not elements") 
        return ItemsByTeam  

    GetRandomItem<private>(ItemsData:[]item_data): item_data=
        var NewItem: item_data = item_data{}
        IndexItemRandon := GetRandomInt(0, ItemsData.Length-1)
        if:
            ItemChoosen := ItemsData[IndexItemRandon]
            RifleData := rifle_data[ItemChoosen]
        then:
            set NewItem = RifleData
        return NewItem	 
    GrantItem<override>(Agent:agent, Index:int):void= 
        Logger.Print("Rifle Granter")
        RifleDefinition.GrantItemToPlayer(Agent, Index)

    GetItemsData<override>():[]item_data =
        RifleDefinition.GetRifleData() 

    HasRarity<override>(Rarity:item_rarity): logic=
        ItemsData := GetItemsData()
        for (ItemData:ItemsData):
            if (ItemData.Rarity = Rarity):
                return true
        return false

    HasRarity<override>(ItemsData:[]item_data,Rarity:item_rarity): logic=        
        for (ItemData:ItemsData):
            if (ItemData.Rarity = Rarity):
                return true
        return false
        
rifle_data := class<concrete>(weapon_data, iconTextureable):
    @editable
    ItemId: item_rifle_id = item_rifle_id.Undefined 

    WeaponType<override>:weapon = weapon.Rifle
    
    GetIconTexture<override>()<computes>: texture = FindRifleIconTexture(ItemId)
    
reward_rifle_definition := class<concrete>:

    @editable # One item granter with all the Rifles
    RifleGranter<private>:item_granter_device = item_granter_device{}	

    GrantItemToPlayer<public>(Agent:agent, Index:int):void= 
        RifleGranter.GrantItemIndex(Agent, Index)

    GetRifleData<public>():[]item_data =
        return RifleData

    @editable
    RifleData<private>:[]rifle_data := array{	
        rifle_data
        {
            ItemId:= item_rifle_id.BurstAssaultRifle_L1,
            DisplayName:= "Burst Assault Rifle",
            Rarity := item_rarity.Common,
            ItemGranterIndex := 0,
            Team := team_id.Blue,
            Attributes := array{         
                item_attribute_data{Id := item_attribute.DPS, Value:= 0.0},
                item_attribute_data{Id := item_attribute.Damage, Value:= 0.0},
                item_attribute_data{Id := item_attribute.MagSize, Value:= 0.0}
            }
        },
        rifle_data
        {
            ItemId:= item_rifle_id.BurstAssaultRifle_L2,
            DisplayName:= "Burst Assault Rifle",
            Rarity := item_rarity.Uncommon,
            ItemGranterIndex := 1,
            Team := team_id.Blue,
            Attributes := array{         
                item_attribute_data{Id := item_attribute.DPS, Value:= 0.0},
                item_attribute_data{Id := item_attribute.Damage, Value:= 0.0},
                item_attribute_data{Id := item_attribute.MagSize, Value:= 0.0}
            }
        },
        rifle_data
        {
            ItemId:= item_rifle_id.BurstAssaultRifle_L3,
            DisplayName:= "Burst Assault Rifle",
            Rarity := item_rarity.Rare,
            ItemGranterIndex := 2,
            Team := team_id.Blue,
            Attributes := array{         
                item_attribute_data{Id := item_attribute.DPS, Value:= 101.5},
                item_attribute_data{Id := item_attribute.Damage, Value:= 25.0},
                item_attribute_data{Id := item_attribute.MagSize, Value:= 30.0}
            }
        },
        rifle_data
        {
            ItemId:= item_rifle_id.BurstAssaultRifle_L4,
            DisplayName:= "Burst Assault Rifle",
            Rarity := item_rarity.Epic,
            ItemGranterIndex := 3,
            Team := team_id.Blue,
            Attributes := array{         
                item_attribute_data{Id := item_attribute.DPS, Value:= 0.0},
                item_attribute_data{Id := item_attribute.Damage, Value:= 0.0},
                item_attribute_data{Id := item_attribute.MagSize, Value:= 0.0}
                }
        },
        rifle_data
        {
            ItemId:= item_rifle_id.BurstAssaultRifle_L5,
            DisplayName:= "Burst Assault Rifle",
            Rarity := item_rarity.Legendary,
            ItemGranterIndex := 4,
            Team := team_id.Blue,
            Attributes := array{         
                item_attribute_data{Id := item_attribute.DPS, Value:= 0.0},
                item_attribute_data{Id := item_attribute.Damage, Value:= 0.0},
                item_attribute_data{Id := item_attribute.MagSize, Value:= 0.0}
            }
        },
        rifle_data
        {
            ItemId:= item_rifle_id.AssaultRifle_L1,
            DisplayName:= "Assault Rifle",
            Rarity := item_rarity.Common,
            ItemGranterIndex := 5,
            Team := team_id.Blue,
            Attributes := array{         
                item_attribute_data{Id := item_attribute.DPS, Value:= 0.0},
                item_attribute_data{Id := item_attribute.Damage, Value:= 0.0},
                item_attribute_data{Id := item_attribute.MagSize, Value:= 0.0}
            }
        },
        rifle_data
        {
            ItemId:= item_rifle_id.AssaultRifle_L2,
            DisplayName:= "Assault Rifle",
            Rarity := item_rarity.Uncommon,
            ItemGranterIndex := 6,
            Team := team_id.Blue,
            Attributes := array{         
                item_attribute_data{Id := item_attribute.DPS, Value:= 0.0},
                item_attribute_data{Id := item_attribute.Damage, Value:= 0.0},
                item_attribute_data{Id := item_attribute.MagSize, Value:= 0.0}
            }
        },
        rifle_data
        {
            ItemId:= item_rifle_id.AssaultRifle_L3,
            DisplayName:= "Assault Rifle",
            Rarity := item_rarity.Rare,
            ItemGranterIndex := 7,
            Team := team_id.Blue,
            Attributes := array{         
                item_attribute_data{Id := item_attribute.DPS, Value:= 193.6},
                item_attribute_data{Id := item_attribute.Damage, Value:= 22.0},
                item_attribute_data{Id := item_attribute.MagSize, Value:= 28.0}
            }
        },
        rifle_data
        {
            ItemId:= item_rifle_id.AssaultRifle_L4,
            DisplayName:= "Assault Rifle",
            Rarity := item_rarity.Epic,
            ItemGranterIndex := 8,
            Team := team_id.Blue,
            Attributes := array{         
                item_attribute_data{Id := item_attribute.DPS, Value:= 0.0},
                item_attribute_data{Id := item_attribute.Damage, Value:= 0.0},
                item_attribute_data{Id := item_attribute.MagSize, Value:= 0.0}
            }
        },
        rifle_data
        {
            ItemId:= item_rifle_id.AssaultRifle_L5,
            DisplayName:= "Assault Rifle",
            Rarity := item_rarity.Legendary,
            ItemGranterIndex := 9,
            Team := team_id.Blue,
            Attributes := array{         
                item_attribute_data{Id := item_attribute.DPS, Value:= 0.0},
                item_attribute_data{Id := item_attribute.Damage, Value:= 0.0},
                item_attribute_data{Id := item_attribute.MagSize, Value:= 0.0}
            }
        },
        rifle_data
        {
            ItemId:= item_rifle_id.HeavyAssaultRifle_L1,
            DisplayName:= "Heavy Assault Rifle",
            Rarity := item_rarity.Common,
            ItemGranterIndex := 10,
            Team := team_id.Blue,
            Attributes := array{         
                item_attribute_data{Id := item_attribute.DPS, Value:= 0.0},
                item_attribute_data{Id := item_attribute.Damage, Value:= 0.0},
                item_attribute_data{Id := item_attribute.MagSize, Value:= 0.0}
            }
        },
        rifle_data
        {
            ItemId:= item_rifle_id.HeavyAssaultRifle_L2,
            DisplayName:= "Heavy Assault Rifle",
            Rarity := item_rarity.Uncommon,
            ItemGranterIndex := 11,
            Team := team_id.Blue,
            Attributes := array{         
                item_attribute_data{Id := item_attribute.DPS, Value:= 0.0},
                item_attribute_data{Id := item_attribute.Damage, Value:= 0.0},
                item_attribute_data{Id := item_attribute.MagSize, Value:= 0.0}
            }
        },
        rifle_data
        {
            ItemId:= item_rifle_id.HeavyAssaultRifle_L3,
            DisplayName:= "Heavy Assault Rifle",
            Rarity := item_rarity.Rare,
            ItemGranterIndex := 12,
            Team := team_id.Blue,
            Attributes := array{         
                item_attribute_data{Id := item_attribute.DPS, Value:= 142.5},
                item_attribute_data{Id := item_attribute.Damage, Value:= 38.0},
                item_attribute_data{Id := item_attribute.MagSize, Value:= 25.0}
            }
        },
        rifle_data
        {
            ItemId:= item_rifle_id.HeavyAssaultRifle_L4,
            DisplayName:= "Heavy Assault Rifle",
            Rarity := item_rarity.Epic,
            ItemGranterIndex := 13,
            Team := team_id.Blue,
            Attributes := array{         
                item_attribute_data{Id := item_attribute.DPS, Value:= 0.0},
                item_attribute_data{Id := item_attribute.Damage, Value:= 0.0},
                item_attribute_data{Id := item_attribute.MagSize, Value:= 0.0}
            }
        },
        rifle_data
        {
            ItemId:= item_rifle_id.HeavyAssaultRifle_L5,
            DisplayName:= "Heavy Assault Rifle",
            Rarity := item_rarity.Legendary,
            ItemGranterIndex := 14,
            Team := team_id.Blue,
            Attributes := array{         
                item_attribute_data{Id := item_attribute.DPS, Value:= 0.0},
                item_attribute_data{Id := item_attribute.Damage, Value:= 0.0},
                item_attribute_data{Id := item_attribute.MagSize, Value:= 0.0}
            }
        },
        rifle_data
        {
            ItemId:= item_rifle_id.InfantryRifle_L1,
            DisplayName:= "Infantry Rifle",
            Rarity := item_rarity.Common,
            ItemGranterIndex := 15,
            Team := team_id.Blue,
            Attributes := array{         
                item_attribute_data{Id := item_attribute.DPS, Value:= 0.0},
                item_attribute_data{Id := item_attribute.Damage, Value:= 0.0},
                item_attribute_data{Id := item_attribute.MagSize, Value:= 0.0}
            }
        },
        rifle_data
        {
            ItemId:= item_rifle_id.InfantryRifle_L2,
            DisplayName:= "Infantry Rifle",
            Rarity := item_rarity.Uncommon,
            ItemGranterIndex := 16,
            Team := team_id.Blue,
            Attributes := array{         
                item_attribute_data{Id := item_attribute.DPS, Value:= 0.0},
                item_attribute_data{Id := item_attribute.Damage, Value:= 0.0},
                item_attribute_data{Id := item_attribute.MagSize, Value:= 0.0}
            }
        },
        rifle_data
        {
            ItemId:= item_rifle_id.InfantryRifle_L3,
            DisplayName:= "Infantry Rifle",
            Rarity := item_rarity.Rare,
            ItemGranterIndex := 17,
            Team := team_id.Blue,
            Attributes := array{         
                item_attribute_data{Id := item_attribute.DPS, Value:= 160.0},
                item_attribute_data{Id := item_attribute.Damage, Value:= 40.0},
                item_attribute_data{Id := item_attribute.MagSize, Value:= 8.0}
            }
        },
        rifle_data
        {
            ItemId:= item_rifle_id.InfantryRifle_L4,
            DisplayName:= "Infantry Rifle",
            Rarity := item_rarity.Epic,
            ItemGranterIndex := 18,
            Team := team_id.Blue,
            Attributes := array{         
                item_attribute_data{Id := item_attribute.DPS, Value:= 0.0},
                item_attribute_data{Id := item_attribute.Damage, Value:= 0.0},
                item_attribute_data{Id := item_attribute.MagSize, Value:= 0.0}
            }
        },
        rifle_data
        {
            ItemId:= item_rifle_id.InfantryRifle_L5,
            DisplayName:= "Infantry Rifle",
            Rarity := item_rarity.Legendary,
            ItemGranterIndex := 19,
            Team := team_id.Blue,
            Attributes := array{         
                item_attribute_data{Id := item_attribute.DPS, Value:= 0.0},
                item_attribute_data{Id := item_attribute.Damage, Value:= 0.0},
                item_attribute_data{Id := item_attribute.MagSize, Value:= 0.0}
            }
        },
        rifle_data
        {
            ItemId:= item_rifle_id.RangerAssaultRifle_L1,
            DisplayName:= "Ranger Assault Rifle",
            Rarity := item_rarity.Common,
            ItemGranterIndex := 20,
            Team := team_id.Blue,
            Attributes := array{         
                item_attribute_data{Id := item_attribute.DPS, Value:= 0.0},
                item_attribute_data{Id := item_attribute.Damage, Value:= 0.0},
                item_attribute_data{Id := item_attribute.MagSize, Value:= 0.0}
                }
        },
        rifle_data
        {
            ItemId:= item_rifle_id.RangerAssaultRifle_L2,
            DisplayName:= "Ranger Assault Rifle",
            Rarity := item_rarity.Uncommon,
            ItemGranterIndex := 21,
            Team := team_id.Blue,
            Attributes := array{         
                item_attribute_data{Id := item_attribute.DPS, Value:= 0.0},
                item_attribute_data{Id := item_attribute.Damage, Value:= 0.0},
                item_attribute_data{Id := item_attribute.MagSize, Value:= 0.0}
            }
        },
        rifle_data
        {
            ItemId:= item_rifle_id.RangerAssaultRifle_L3,
            DisplayName:= "Ranger Assault Rifle",
            Rarity := item_rarity.Rare,
            ItemGranterIndex := 22,
            Team := team_id.Blue,
            Attributes := array{         
                item_attribute_data{Id := item_attribute.DPS, Value:= 136.0},
                item_attribute_data{Id := item_attribute.Damage, Value:= 34.0},
                item_attribute_data{Id := item_attribute.MagSize, Value:= 25.0}
            }
        },
        rifle_data
        {
            ItemId:= item_rifle_id.RangerAssaultRifle_L4,
            DisplayName:= "Ranger Assault Rifle",
            Rarity := item_rarity.Epic,
            ItemGranterIndex := 23,
            Team := team_id.Blue,
            Attributes := array{         
                item_attribute_data{Id := item_attribute.DPS, Value:= 0.0},
                item_attribute_data{Id := item_attribute.Damage, Value:= 0.0},
                item_attribute_data{Id := item_attribute.MagSize, Value:= 0.0}
            }
        },
        rifle_data
        {
            ItemId:= item_rifle_id.RangerAssaultRifle_L5,
            DisplayName:= "Ranger Assault Rifle",
            Rarity := item_rarity.Legendary,
            ItemGranterIndex := 24,
            Team := team_id.Blue,
            Attributes := array{         
                item_attribute_data{Id := item_attribute.DPS, Value:= 0.0},
                item_attribute_data{Id := item_attribute.Damage, Value:= 0.0},
                item_attribute_data{Id := item_attribute.MagSize, Value:= 0.0}
            }
        },
        rifle_data
        {
            ItemId:= item_rifle_id.RangerAssaultRifle_L6,
            DisplayName:= "Ranger Assault Rifle",
            Rarity := item_rarity.Mythic,
            ItemGranterIndex := 25,
            Team := team_id.Blue,
            Attributes := array{         
                item_attribute_data{Id := item_attribute.DPS, Value:= 0.0},
                item_attribute_data{Id := item_attribute.Damage, Value:= 0.0},
                item_attribute_data{Id := item_attribute.MagSize, Value:= 0.0}
            }
        },
        rifle_data
        {
            ItemId:= item_rifle_id.WarforgedAssaultRifle_L1,
            DisplayName:= "Warforged Assault Rifle",
            Rarity := item_rarity.Common,
            ItemGranterIndex := 26,
            Team := team_id.Blue,
            Attributes := array{         
                item_attribute_data{Id := item_attribute.DPS, Value:= 0.0},
                item_attribute_data{Id := item_attribute.Damage, Value:= 0.0},
                item_attribute_data{Id := item_attribute.MagSize, Value:= 0.0}
            }
        },
        rifle_data
        {
            ItemId:= item_rifle_id.WarforgedAssaultRifle_L2,
            DisplayName:= "Warforged Assault Rifle",
            Rarity := item_rarity.Uncommon,
            ItemGranterIndex := 27,
            Team := team_id.Blue,
            Attributes := array{         
                item_attribute_data{Id := item_attribute.DPS, Value:= 0.0},
                item_attribute_data{Id := item_attribute.Damage, Value:= 0.0},
                item_attribute_data{Id := item_attribute.MagSize, Value:= 0.0}
            }
        },
        rifle_data
        {
            ItemId:= item_rifle_id.WarforgedAssaultRifle_L3,
            DisplayName:= "Warforged Assault Rifle",
            Rarity := item_rarity.Rare,
            ItemGranterIndex := 28,
            Team := team_id.Blue,
            Attributes := array{         
                item_attribute_data{Id := item_attribute.DPS, Value:= 194.4},
                item_attribute_data{Id := item_attribute.Damage, Value:= 27.0},
                item_attribute_data{Id := item_attribute.MagSize, Value:= 35.0}
            }
        },
        rifle_data
        {
            ItemId:= item_rifle_id.WarforgedAssaultRifle_L4,
            DisplayName:= "Warforged Assault Rifle",
            Rarity := item_rarity.Epic,
            ItemGranterIndex := 29,
            Team := team_id.Blue,
            Attributes := array{         
                item_attribute_data{Id := item_attribute.DPS, Value:= 0.0},
                item_attribute_data{Id := item_attribute.Damage, Value:= 0.0},
                item_attribute_data{Id := item_attribute.MagSize, Value:= 0.0}
            }
        },
        rifle_data
        {
            ItemId:= item_rifle_id.WarforgedAssaultRifle_L5,
            DisplayName:= "Warforged Assault Rifle",
            Rarity := item_rarity.Legendary,
            ItemGranterIndex := 30,
            Team := team_id.Blue,
            Attributes := array{         
                item_attribute_data{Id := item_attribute.DPS, Value:= 0.0},
                item_attribute_data{Id := item_attribute.Damage, Value:= 0.0},
                item_attribute_data{Id := item_attribute.MagSize, Value:= 0.0}
            }
        }
    }