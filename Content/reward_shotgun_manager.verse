using { /Fortnite.com/Devices }
using { /Verse.org/Random }
using { /Verse.org/Assets }
using { /Verse.org/Simulation }
using { /UnrealEngine.com/Temporary/Diagnostics }

log_reward_shotgun_manager := class(log_channel){}

reward_shotgun_manager := class(creative_device,grantable):

    Logger:log = log{Channel:=log_reward_shotgun_manager}

    @editable
    ShotgunDefinition<private>:reward_shotgun_definition = reward_shotgun_definition{}  

    FindItemByRarity<override>(ItemsData:[]item_data, Rarity:item_rarity):item_data =
        Logger.Print("Shotgun Manager")
        var ItemsByRarity:[]item_data = array{}
        var ItemData:item_data = item_data{}  
        for (TempItemData:ItemsData):
            if (TempItemData.Rarity = Rarity):
                set ItemsByRarity += array{TempItemData} 
        if:
            ItemsByRarity.Length > 0            
        then:
            set ItemData = GetRandomItem(ItemsByRarity)       
        return ItemData 

    FindItemsByTeam<override>(ItemsData:[]item_data, Team:team_id):[]item_data =
        var ItemsByTeam:[]item_data = array{}
        for (ItemData:ItemsData):
            if (ItemData.Team = Team):
                set ItemsByTeam += array{ItemData} 
        if:
            ItemsByTeam.Length <= 0            
        then:
            Logger.Print("There are not elements") 
        return ItemsByTeam  

    GetRandomItem<private>(ItemsData:[]item_data): item_data=
        var NewItem: item_data = item_data{}
        IndexItemRandon := GetRandomInt(0, ItemsData.Length-1)
        if:
            ItemChoosen := ItemsData[IndexItemRandon]
            ShotgunData := shotgun_data[ItemChoosen]
        then:
            set NewItem = ShotgunData
        return NewItem	 

    GrantItem<override>(Agent:agent, Index:int):void= 
        Logger.Print("Shotgun Granter")
        ShotgunDefinition.GrantItemToPlayer(Agent, Index)

    GetItemsData<override>():[]item_data =
        ShotgunDefinition.GetShotgunData() 

    HasRarity<override>(Rarity:item_rarity): logic=
        ItemsData := GetItemsData()
        for (ItemData:ItemsData):
            if (ItemData.Rarity = Rarity):
                return true
        return false

    HasRarity<override>(ItemsData:[]item_data,Rarity:item_rarity): logic=        
        for (ItemData:ItemsData):
            if (ItemData.Rarity = Rarity):
                return true
        return false

shotgun_data := class<concrete>(weapon_data, iconTextureable):
    @editable
    ItemId: item_shotgun_id = item_shotgun_id.Undefined    
    
    WeaponType<override>:weapon = weapon.Shotgun
    
    GetIconTexture<override>()<computes>: texture = FindShotgunIconTexture(ItemId)

reward_shotgun_definition := class<concrete>:

    @editable # One item granter with all the Shotguns
    ShotgunGranter<private>:item_granter_device = item_granter_device{}	

    GrantItemToPlayer<public>(Agent:agent, Index:int):void= 	
        ShotgunGranter.GrantItemIndex(Agent, Index)

    GetShotgunData<public>():[]item_data =
        return ShotgunData

    @editable
    ShotgunData<private>:[]shotgun_data := array{	
        shotgun_data
        {
            ItemId:= item_shotgun_id.AutoShotgun_L1,
            DisplayName:= "Auto Shotgun",
            Rarity := item_rarity.Common,
            ItemGranterIndex := 0,
            Team := team_id.Blue,
            Attributes := array{         
                item_attribute_data{Id := item_attribute.DPS, Value:= 0.0},
            item_attribute_data{Id := item_attribute.Damage, Value:= 0.0},
            item_attribute_data{Id := item_attribute.MagSize, Value:= 0.0}
            }
        },
    shotgun_data
        {
            ItemId:= item_shotgun_id.AutoShotgun_L2,
            DisplayName:= "Auto Shotgun",
            Rarity := item_rarity.Uncommon,
            ItemGranterIndex := 1,
            Team := team_id.Blue,
            Attributes := array{         
                item_attribute_data{Id := item_attribute.DPS, Value:= 0.0},
                item_attribute_data{Id := item_attribute.Damage, Value:= 0.0},
                item_attribute_data{Id := item_attribute.MagSize, Value:= 0.0}
            }
        },
    shotgun_data
        {
            ItemId:= item_shotgun_id.AutoShotgun_L3,
            DisplayName:= "Auto Shotgun",
            Rarity := item_rarity.Rare,
            ItemGranterIndex := 2,
            Team := team_id.Blue,
            Attributes := array{         
                item_attribute_data{Id := item_attribute.DPS, Value:= 140.4},
                item_attribute_data{Id := item_attribute.Damage, Value:= 93.6},
                item_attribute_data{Id := item_attribute.MagSize, Value:= 8.0}
            }
        },
    shotgun_data
        {
            ItemId:= item_shotgun_id.AutoShotgun_L4,
            DisplayName:= "Auto Shotgun",
            Rarity := item_rarity.Epic,
            ItemGranterIndex := 3,
            Team := team_id.Blue,
            Attributes := array{         
                item_attribute_data{Id := item_attribute.DPS, Value:= 0.0},
                item_attribute_data{Id := item_attribute.Damage, Value:= 0.0},
                item_attribute_data{Id := item_attribute.MagSize, Value:= 0.0}
            }
        },
    shotgun_data
        {
            ItemId:= item_shotgun_id.AutoShotgun_L5,
            DisplayName:= "Auto Shotgun",
            Rarity := item_rarity.Legendary,
            ItemGranterIndex := 4,
            Team := team_id.Blue,
            Attributes := array{         
                item_attribute_data{Id := item_attribute.DPS, Value:= 0.0},
                item_attribute_data{Id := item_attribute.Damage, Value:= 0.0},
                item_attribute_data{Id := item_attribute.MagSize, Value:= 0.0}
            }
        },
    shotgun_data
        {
            ItemId:= item_shotgun_id.AutoShotgun_L6,
            DisplayName:= "Auto Shotgun",
            Rarity := item_rarity.Mythic,
            ItemGranterIndex := 5,
            Team := team_id.Blue,
            Attributes := array{         
                item_attribute_data{Id := item_attribute.DPS, Value:= 0.0},
                item_attribute_data{Id := item_attribute.Damage, Value:= 0.0},
                item_attribute_data{Id := item_attribute.MagSize, Value:= 0.0}
            }
        },
    shotgun_data
        {
            ItemId:= item_shotgun_id.CombatShotgun_L1,
            DisplayName:= "Combat Shotgun",
            Rarity := item_rarity.Common,
            ItemGranterIndex := 6,
            Team := team_id.Blue,
            Attributes := array{         
                item_attribute_data{Id := item_attribute.DPS, Value:= 0.0},
                item_attribute_data{Id := item_attribute.Damage, Value:= 0.0},
                item_attribute_data{Id := item_attribute.MagSize, Value:= 0.0}
            }
        },
    shotgun_data
        {
            ItemId:= item_shotgun_id.CombatShotgun_L2,
            DisplayName:= "Combat Shotgun",
            Rarity := item_rarity.Uncommon,
            ItemGranterIndex := 7,
            Team := team_id.Blue,
            Attributes := array{         
                item_attribute_data{Id := item_attribute.DPS, Value:= 0.0},
                item_attribute_data{Id := item_attribute.Damage, Value:= 0.0},
                item_attribute_data{Id := item_attribute.MagSize, Value:= 0.0}
            }
        },
    shotgun_data
        {
            ItemId:= item_shotgun_id.CombatShotgun_L3,
            DisplayName:= "Combat Shotgun",
            Rarity := item_rarity.Rare,
            ItemGranterIndex := 8,
            Team := team_id.Blue,
            Attributes := array{         
                item_attribute_data{Id := item_attribute.DPS, Value:= 115.6},
                item_attribute_data{Id := item_attribute.Damage, Value:= 68.0},
                item_attribute_data{Id := item_attribute.MagSize, Value:= 8.0}
            }
        },
    shotgun_data
        {
            ItemId:= item_shotgun_id.CombatShotgun_L4,
            DisplayName:= "Combat Shotgun",
            Rarity := item_rarity.Epic,
            ItemGranterIndex := 9,
            Team := team_id.Blue,
            Attributes := array{         
                item_attribute_data{Id := item_attribute.DPS, Value:= 0.0},
                item_attribute_data{Id := item_attribute.Damage, Value:= 0.0},
                item_attribute_data{Id := item_attribute.MagSize, Value:= 0.0}
            }
        },
    shotgun_data
        {
            ItemId:= item_shotgun_id.CombatShotgun_L5,
            DisplayName:= "Combat Shotgun",
            Rarity := item_rarity.Legendary,
            ItemGranterIndex := 10,
            Team := team_id.Blue,
            Attributes := array{         
                item_attribute_data{Id := item_attribute.DPS, Value:= 0.0},
                item_attribute_data{Id := item_attribute.Damage, Value:= 0.0},
                item_attribute_data{Id := item_attribute.MagSize, Value:= 0.0}
            }
        },
    shotgun_data
        {
            ItemId:= item_shotgun_id.DrumShotgun_L1,
            DisplayName:= "Drum Shotgun",
            Rarity := item_rarity.Common,
            ItemGranterIndex := 11,
            Team := team_id.Blue,
            Attributes := array{         
                item_attribute_data{Id := item_attribute.DPS, Value:= 0.0},
                item_attribute_data{Id := item_attribute.Damage, Value:= 0.0},
                item_attribute_data{Id := item_attribute.MagSize, Value:= 0.0}
            }
        },
    shotgun_data
        {
            ItemId:= item_shotgun_id.DrumShotgun_L2,
            DisplayName:= "Drum Shotgun",
            Rarity := item_rarity.Uncommon,
            ItemGranterIndex := 12,
            Team := team_id.Blue,
            Attributes := array{         
                item_attribute_data{Id := item_attribute.DPS, Value:= 0.0},
                item_attribute_data{Id := item_attribute.Damage, Value:= 0.0},
                item_attribute_data{Id := item_attribute.MagSize, Value:= 0.0}
            }
        },
    shotgun_data
        {
            ItemId:= item_shotgun_id.DrumShotgun_L3,
            DisplayName:= "Drum Shotgun",
            Rarity := item_rarity.Rare,
            ItemGranterIndex := 13,
            Team := team_id.Blue,
            Attributes := array{         
                item_attribute_data{Id := item_attribute.DPS, Value:= 169.2},
                item_attribute_data{Id := item_attribute.Damage, Value:= 56.4},
                item_attribute_data{Id := item_attribute.MagSize, Value:= 12.0}
            }
        },
    shotgun_data
        {
            ItemId:= item_shotgun_id.DrumShotgun_L4,
            DisplayName:= "Drum Shotgun",
            Rarity := item_rarity.Epic,
            ItemGranterIndex := 14,
            Team := team_id.Blue,
            Attributes := array{         
                item_attribute_data{Id := item_attribute.DPS, Value:= 0.0},
                item_attribute_data{Id := item_attribute.Damage, Value:= 0.0},
                item_attribute_data{Id := item_attribute.MagSize, Value:= 0.0}
            }
        },
    shotgun_data
        {
            ItemId:= item_shotgun_id.DrumShotgun_L5,
            DisplayName:= "Drum Shotgun",
            Rarity := item_rarity.Legendary,
            ItemGranterIndex := 15,
            Team := team_id.Blue,
            Attributes := array{         
                item_attribute_data{Id := item_attribute.DPS, Value:= 0.0},
                item_attribute_data{Id := item_attribute.Damage, Value:= 0.0},
                item_attribute_data{Id := item_attribute.MagSize, Value:= 0.0}
            }
        },
    shotgun_data
        {
            ItemId:= item_shotgun_id.GatekeeperShotgun_L1,
            DisplayName:= "Gatekeeper Shotgun",
            Rarity := item_rarity.Common,
            ItemGranterIndex := 16,
            Team := team_id.Blue,
            Attributes := array{         
                item_attribute_data{Id := item_attribute.DPS, Value:= 0.0},
                item_attribute_data{Id := item_attribute.Damage, Value:= 0.0},
                item_attribute_data{Id := item_attribute.MagSize, Value:= 0.0}
            }
        },
    shotgun_data
        {
            ItemId:= item_shotgun_id.GatekeeperShotgun_L2,
            DisplayName:= "Gatekeeper Shotgun",
            Rarity := item_rarity.Uncommon,
            ItemGranterIndex := 17,
            Team := team_id.Blue,
            Attributes := array{         
                item_attribute_data{Id := item_attribute.DPS, Value:= 0.0},
                item_attribute_data{Id := item_attribute.Damage, Value:= 0.0},
                item_attribute_data{Id := item_attribute.MagSize, Value:= 0.0}
            }
        },
    shotgun_data
        {
            ItemId:= item_shotgun_id.GatekeeperShotgun_L3,
            DisplayName:= "Gatekeeper Shotgun",
            Rarity := item_rarity.Rare,
            ItemGranterIndex := 18,
            Team := team_id.Blue,
            Attributes := array{         
                item_attribute_data{Id := item_attribute.DPS, Value:= 142.8},
                item_attribute_data{Id := item_attribute.Damage, Value:= 84.0},
                item_attribute_data{Id := item_attribute.MagSize, Value:= 3.0}
            }
        },
    shotgun_data
        {
            ItemId:= item_shotgun_id.GatekeeperShotgun_L4,
            DisplayName:= "Gatekeeper Shotgun",
            Rarity := item_rarity.Epic,
            ItemGranterIndex := 19,
            Team := team_id.Blue,
            Attributes := array{         
                item_attribute_data{Id := item_attribute.DPS, Value:= 0.0},
                item_attribute_data{Id := item_attribute.Damage, Value:= 0.0},
                item_attribute_data{Id := item_attribute.MagSize, Value:= 0.0}
            }
        },
    shotgun_data
        {
            ItemId:= item_shotgun_id.GatekeeperShotgun_L5,
            DisplayName:= "Gatekeeper Shotgun",
            Rarity := item_rarity.Legendary,
            ItemGranterIndex := 20,
            Team := team_id.Blue,
            Attributes := array{         
                item_attribute_data{Id := item_attribute.DPS, Value:= 0.0},
                item_attribute_data{Id := item_attribute.Damage, Value:= 0.0},
                item_attribute_data{Id := item_attribute.MagSize, Value:= 0.0}
            }
        },
    shotgun_data
        {
            ItemId:= item_shotgun_id.SovereignShotgun_L1,
            DisplayName:= "Sovereign Shotgun",
            Rarity := item_rarity.Common,
            ItemGranterIndex := 21,
            Team := team_id.Blue,
            Attributes := array{         
                item_attribute_data{Id := item_attribute.DPS, Value:= 0.0},
                item_attribute_data{Id := item_attribute.Damage, Value:= 0.0},
                item_attribute_data{Id := item_attribute.MagSize, Value:= 0.0}
            }
        },
    shotgun_data
        {
            ItemId:= item_shotgun_id.SovereignShotgun_L2,
            DisplayName:= "Sovereign Shotgun",
            Rarity := item_rarity.Uncommon,
            ItemGranterIndex := 22,
            Team := team_id.Blue,
            Attributes := array{         
                item_attribute_data{Id := item_attribute.DPS, Value:= 0.0},
                item_attribute_data{Id := item_attribute.Damage, Value:= 0.0},
                item_attribute_data{Id := item_attribute.MagSize, Value:= 0.0}
            }
        },
    shotgun_data
        {
            ItemId:= item_shotgun_id.SovereignShotgun_L3,
            DisplayName:= "Sovereign Shotgun",
            Rarity := item_rarity.Rare,
            ItemGranterIndex := 23,
            Team := team_id.Blue,
            Attributes := array{         
                item_attribute_data{Id := item_attribute.DPS, Value:= 132.5},
                item_attribute_data{Id := item_attribute.Damage, Value:= 92.0},
                item_attribute_data{Id := item_attribute.MagSize, Value:= 6.0}
            }
        },
    shotgun_data
        {
            ItemId:= item_shotgun_id.SovereignShotgun_L4,
            DisplayName:= "Sovereign Shotgun",
            Rarity := item_rarity.Epic,
            ItemGranterIndex := 24,
            Team := team_id.Blue,
            Attributes := array{         
                item_attribute_data{Id := item_attribute.DPS, Value:= 0.0},
                item_attribute_data{Id := item_attribute.Damage, Value:= 0.0},
                item_attribute_data{Id := item_attribute.MagSize, Value:= 0.0}
            }
        },
    shotgun_data
        {
            ItemId:= item_shotgun_id.SovereignShotgun_L5,
            DisplayName:= "Sovereign Shotgun",
            Rarity := item_rarity.Legendary,
            ItemGranterIndex := 25,
            Team := team_id.Blue,
            Attributes := array{         
                item_attribute_data{Id := item_attribute.DPS, Value:= 0.0},
                item_attribute_data{Id := item_attribute.Damage, Value:= 0.0},
                item_attribute_data{Id := item_attribute.MagSize, Value:= 0.0}
            }
        },
    shotgun_data
        {
            ItemId:= item_shotgun_id.LeverActionShotgun_L1,
            DisplayName:= "Lever Action Shotgun",
            Rarity := item_rarity.Common,
            ItemGranterIndex := 26,
            Team := team_id.Blue,
            Attributes := array{         
                item_attribute_data{Id := item_attribute.DPS, Value:= 0.0},
                item_attribute_data{Id := item_attribute.Damage, Value:= 0.0},
                item_attribute_data{Id := item_attribute.MagSize, Value:= 0.0}
            }
        },
    shotgun_data
        {
            ItemId:= item_shotgun_id.LeverActionShotgun_L2,
            DisplayName:= "Lever Action Shotgun",
            Rarity := item_rarity.Uncommon,
            ItemGranterIndex := 27,
            Team := team_id.Blue,
            Attributes := array{         
                item_attribute_data{Id := item_attribute.DPS, Value:= 0.0},
                item_attribute_data{Id := item_attribute.Damage, Value:= 0.0},
                item_attribute_data{Id := item_attribute.MagSize, Value:= 0.0}
            }
        },
    shotgun_data
        {
            ItemId:= item_shotgun_id.LeverActionShotgun_L3,
            DisplayName:= "Lever Action Shotgun",
            Rarity := item_rarity.Rare,
            ItemGranterIndex := 28,
            Team := team_id.Blue,
            Attributes := array{         
                item_attribute_data{Id := item_attribute.DPS, Value:= 0.0},
                item_attribute_data{Id := item_attribute.Damage, Value:= 0.0},
                item_attribute_data{Id := item_attribute.MagSize, Value:= 0.0}
            }
        },
    shotgun_data
        {
            ItemId:= item_shotgun_id.LeverActionShotgun_L4,
            DisplayName:= "Lever Action Shotgun",
            Rarity := item_rarity.Epic,
            ItemGranterIndex := 29,
            Team := team_id.Blue,
            Attributes := array{         
                item_attribute_data{Id := item_attribute.DPS, Value:= 0.0},
                item_attribute_data{Id := item_attribute.Damage, Value:= 0.0},
                item_attribute_data{Id := item_attribute.MagSize, Value:= 0.0}
            }
        },
    shotgun_data
        {
            ItemId:= item_shotgun_id.LeverActionShotgun_L5,
            DisplayName:= "Lever Action Shotgun",
            Rarity := item_rarity.Legendary,
            ItemGranterIndex := 30,
            Team := team_id.Blue,
            Attributes := array{         
                item_attribute_data{Id := item_attribute.DPS, Value:= 0.0},
                item_attribute_data{Id := item_attribute.Damage, Value:= 0.0},
                item_attribute_data{Id := item_attribute.MagSize, Value:= 0.0}
            }
        },
    shotgun_data
        {
            ItemId:= item_shotgun_id.PumpShotgun_L1,
            DisplayName:= "Pump Shotgun",
            Rarity := item_rarity.Common,
            ItemGranterIndex := 31,
            Team := team_id.Blue,
            Attributes := array{         
                item_attribute_data{Id := item_attribute.DPS, Value:= 0.0},
                item_attribute_data{Id := item_attribute.Damage, Value:= 0.0},
                item_attribute_data{Id := item_attribute.MagSize, Value:= 0.0}
            }
        },
    shotgun_data
        {
            ItemId:= item_shotgun_id.PumpShotgun_L2,
            DisplayName:= "Pump Shotgun",
            Rarity := item_rarity.Uncommon,
            ItemGranterIndex := 32,
            Team := team_id.Blue,
            Attributes := array{         
                item_attribute_data{Id := item_attribute.DPS, Value:= 0.0},
                item_attribute_data{Id := item_attribute.Damage, Value:= 0.0},
                item_attribute_data{Id := item_attribute.MagSize, Value:= 0.0}
            }
        },
    shotgun_data
        {
            ItemId:= item_shotgun_id.PumpShotgun_L3,
            DisplayName:= "Pump Shotgun",
            Rarity := item_rarity.Rare,
            ItemGranterIndex := 33,
            Team := team_id.Blue,
            Attributes := array{         
                item_attribute_data{Id := item_attribute.DPS, Value:= 0.0},
                item_attribute_data{Id := item_attribute.Damage, Value:= 0.0},
                item_attribute_data{Id := item_attribute.MagSize, Value:= 0.0}
            }
        },
    shotgun_data
        {
            ItemId:= item_shotgun_id.PumpShotgun_L4,
            DisplayName:= "Pump Shotgun",
            Rarity := item_rarity.Epic,
            ItemGranterIndex := 34,
            Team := team_id.Blue,
            Attributes := array{         
                item_attribute_data{Id := item_attribute.DPS, Value:= 0.0},
                item_attribute_data{Id := item_attribute.Damage, Value:= 0.0},
                item_attribute_data{Id := item_attribute.MagSize, Value:= 0.0}
            }
        },
    shotgun_data
        {
            ItemId:= item_shotgun_id.PumpShotgun_L5,
            DisplayName:= "Pump Shotgun",
            Rarity := item_rarity.Legendary,
            ItemGranterIndex := 35,
            Team := team_id.Blue,
            Attributes := array{         
                item_attribute_data{Id := item_attribute.DPS, Value:= 0.0},
                item_attribute_data{Id := item_attribute.Damage, Value:= 0.0},
                item_attribute_data{Id := item_attribute.MagSize, Value:= 0.0}
            }
        },
    shotgun_data
        {
            ItemId:= item_shotgun_id.DoubleBarrelShotgun_L4,
            DisplayName:= "Double Barrel Shotgun",
            Rarity := item_rarity.Epic,
            ItemGranterIndex := 36,
            Team := team_id.Blue,
            Attributes := array{         
                item_attribute_data{Id := item_attribute.DPS, Value:= 0.0},
                item_attribute_data{Id := item_attribute.Damage, Value:= 0.0},
                item_attribute_data{Id := item_attribute.MagSize, Value:= 0.0}
            }
        },
    shotgun_data
        {
            ItemId:= item_shotgun_id.DoubleBarrelShotgun_L5,
            DisplayName:= "Double Barrel Shotgun",
            Rarity := item_rarity.Legendary,
            ItemGranterIndex := 37,
            Team := team_id.Blue,
            Attributes := array{         
                item_attribute_data{Id := item_attribute.DPS, Value:= 0.0},
                item_attribute_data{Id := item_attribute.Damage, Value:= 0.0},
                item_attribute_data{Id := item_attribute.MagSize, Value:= 0.0}
            }
        },
    shotgun_data
        {
            ItemId:= item_shotgun_id.MakeshiftShotgun_L1,
            DisplayName:= "Makeshift Shotgun",
            Rarity := item_rarity.Common,
            ItemGranterIndex := 38,
            Team := team_id.Blue,
            Attributes := array{         
                item_attribute_data{Id := item_attribute.DPS, Value:= 0.0},
                item_attribute_data{Id := item_attribute.Damage, Value:= 0.0},
                item_attribute_data{Id := item_attribute.MagSize, Value:= 0.0}
            }
        },
    shotgun_data
        {
            ItemId:= item_shotgun_id.MakeshiftShotgun_L2,
            DisplayName:= "Makeshift Shotgun",
            Rarity := item_rarity.Uncommon,
            ItemGranterIndex := 39,
            Team := team_id.Blue,
            Attributes := array{         
                item_attribute_data{Id := item_attribute.DPS, Value:= 0.0},
                item_attribute_data{Id := item_attribute.Damage, Value:= 0.0},
                item_attribute_data{Id := item_attribute.MagSize, Value:= 0.0}
            }
        },
        shotgun_data
        {
            ItemId:= item_shotgun_id.MakeshiftShotgun_L3,
            DisplayName:= "Makeshift Shotgun",
            Rarity := item_rarity.Rare,
            ItemGranterIndex := 40,
            Team := team_id.Blue,
            Attributes := array{         
                item_attribute_data{Id := item_attribute.DPS, Value:= 0.0},
                item_attribute_data{Id := item_attribute.Damage, Value:= 0.0},
                item_attribute_data{Id := item_attribute.MagSize, Value:= 0.0}
            }
        }
    }