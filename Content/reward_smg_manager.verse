using { /Fortnite.com/Devices }
using { /Verse.org/Random }
using { /Verse.org/Assets }
using { /Verse.org/Simulation }
using { /UnrealEngine.com/Temporary/Diagnostics }

log_reward_smg_manager := class(log_channel){}

reward_smg_manager := class(creative_device,grantable):

    Logger:log = log{Channel:=log_reward_smg_manager}

    @editable
    SMGDefinition<private>:reward_smg_definition = reward_smg_definition{} 

    FindItemByRarity<override>(ItemsData:[]item_data, Rarity:item_rarity):item_data =
        Logger.Print("SMG Manager")
        var ItemsByRarity:[]item_data = array{}
        var ItemData:item_data = item_data{}  
        for (TempItemData:ItemsData):
            if (TempItemData.Rarity = Rarity):
                set ItemsByRarity += array{TempItemData} 
        if:
            ItemsByRarity.Length > 0            
        then:
            set ItemData = GetRandomItem(ItemsByRarity)       
        return ItemData 

    FindItemsByTeam<override>(ItemsData:[]item_data, Team:team_id):[]item_data =
        var ItemsByTeam:[]item_data = array{}
        for (ItemData:ItemsData):
            if (ItemData.Team = Team):
                set ItemsByTeam += array{ItemData} 
        if:
            ItemsByTeam.Length <= 0            
        then:
            Logger.Print("There are not elements") 
        return ItemsByTeam  

    GetRandomItem<private>(ItemsData:[]item_data): item_data=
        var NewItem: item_data = item_data{}
        IndexItemRandon := GetRandomInt(0, ItemsData.Length-1)
        if:
            ItemChoosen := ItemsData[IndexItemRandon]
            SMGData := smg_data[ItemChoosen]
        then:
            set NewItem = SMGData
        return NewItem	 

    GrantItem<override>(Agent:agent, Index:int):void= 
        Logger.Print("SMG Granter")
        SMGDefinition.GrantItemToPlayer(Agent, Index)

    GetItemsData<override>():[]item_data =
        SMGDefinition.GetSMGData()

    HasRarity<override>(Rarity:item_rarity): logic=
        ItemsData := GetItemsData()
        for (ItemData:ItemsData):
            if (ItemData.Rarity = Rarity):
                return true
        return false

    HasRarity<override>(ItemsData:[]item_data,Rarity:item_rarity): logic=        
        for (ItemData:ItemsData):
            if (ItemData.Rarity = Rarity):
                return true
        return false

smg_data := class<concrete>(weapon_data, iconTextureable):
    @editable
    ItemId: item_smg_id = item_smg_id.Undefined     

    WeaponType<override>:weapon = weapon.Smg
    
    GetIconTexture<override>()<computes>: texture = FindSMGIconTexture(ItemId)

    GetMaybeAttributes<override>()<computes>:?[]item_attribute_data = option{Attributes}

reward_smg_definition := class<concrete>:
    @editable # One item granter with all the SMGs
    SmgGranter<private>:item_granter_device = item_granter_device{}	

    GrantItemToPlayer<public>(Agent:agent, Index:int):void= 		
        SmgGranter.GrantItemIndex(Agent, Index)

    GetSMGData<public>():[]item_data =
        return SMGData

    @editable
    SMGData<private>:[]smg_data := array{	
        smg_data
        {
            ItemId:= item_smg_id.CombatSMG_L1,
            DisplayName:= "Combat",
            Rarity := item_rarity.Common,
            ItemGranterIndex := 0,
            Team := team_id.Blue,
            Attributes := array{         
                item_attribute_data{Id := item_attribute.DPS, Value:= 0.0},
                item_attribute_data{Id := item_attribute.Damage, Value:= 0.0},
                item_attribute_data{Id := item_attribute.MagSize, Value:= 0.0}
            }
        },
    smg_data
        {
            ItemId:= item_smg_id.CombatSMG_L2,
            DisplayName:= "Combat",
            Rarity := item_rarity.Uncommon,
            ItemGranterIndex := 1,
            Team := team_id.Blue,
            Attributes := array{         
                item_attribute_data{Id := item_attribute.DPS, Value:= 0.0},
                item_attribute_data{Id := item_attribute.Damage, Value:= 0.0},
                item_attribute_data{Id := item_attribute.MagSize, Value:= 0.0}
            }
        },
    smg_data
        {
            ItemId:= item_smg_id.CombatSMG_L3,
            DisplayName:= "Combat",
            Rarity := item_rarity.Rare,
            ItemGranterIndex := 2,
            Team := team_id.Blue,
            Attributes := array{         
                item_attribute_data{Id := item_attribute.DPS, Value:= 240.0},
                item_attribute_data{Id := item_attribute.Damage, Value:= 20.0},
                item_attribute_data{Id := item_attribute.MagSize, Value:= 32.0}
            }
        },
    smg_data
        {
            ItemId:= item_smg_id.CombatSMG_L4,
            DisplayName:= "Combat",
            Rarity := item_rarity.Epic,
            ItemGranterIndex := 3,
            Team := team_id.Blue,
            Attributes := array{         
                item_attribute_data{Id := item_attribute.DPS, Value:= 0.0},
                item_attribute_data{Id := item_attribute.Damage, Value:= 0.0},
                item_attribute_data{Id := item_attribute.MagSize, Value:= 0.0}
            }
        },
    smg_data
        {
            ItemId:= item_smg_id.CombatSMG_L5,
            DisplayName:= "Combat",
            Rarity := item_rarity.Legendary,
            ItemGranterIndex := 4,
            Team := team_id.Blue,
            Attributes := array{         
                item_attribute_data{Id := item_attribute.DPS, Value:= 0.0},
                item_attribute_data{Id := item_attribute.Damage, Value:= 0.0},
                item_attribute_data{Id := item_attribute.MagSize, Value:= 0.0}
            }
        },
    smg_data
        {
            ItemId:= item_smg_id.CombatSMG_L6,
            DisplayName:= "Combat",
            Rarity := item_rarity.Mythic,
            ItemGranterIndex := 5,
            Team := team_id.Blue,
            Attributes := array{         
                item_attribute_data{Id := item_attribute.DPS, Value:= 0.0},
                item_attribute_data{Id := item_attribute.Damage, Value:= 0.0},
                item_attribute_data{Id := item_attribute.MagSize, Value:= 0.0}
            }
        },
    smg_data
        {
            ItemId:= item_smg_id.DualMicroSMGs_L1,
            DisplayName:= "Dual Micro Gs",
            Rarity := item_rarity.Common,
            ItemGranterIndex := 6,
            Team := team_id.Blue,
            Attributes := array{         
                item_attribute_data{Id := item_attribute.DPS, Value:= 0.0},
                item_attribute_data{Id := item_attribute.Damage, Value:= 0.0},
                item_attribute_data{Id := item_attribute.MagSize, Value:= 0.0}
            }
        },
    smg_data
        {
            ItemId:= item_smg_id.DualMicroSMGs_L2,
            DisplayName:= "Dual Micro Gs",
            Rarity := item_rarity.Uncommon,
            ItemGranterIndex := 7,
            Team := team_id.Blue,
            Attributes := array{         
                item_attribute_data{Id := item_attribute.DPS, Value:= 0.0},
                item_attribute_data{Id := item_attribute.Damage, Value:= 0.0},
                item_attribute_data{Id := item_attribute.MagSize, Value:= 0.0}
            }
        },
    smg_data
        {
            ItemId:= item_smg_id.DualMicroSMGs_L3,
            DisplayName:= "Dual Micro Gs",
            Rarity := item_rarity.Rare,
            ItemGranterIndex := 8,
            Team := team_id.Blue,
            Attributes := array{         
                item_attribute_data{Id := item_attribute.DPS, Value:= 197.5},
                item_attribute_data{Id := item_attribute.Damage, Value:= 11.0},
                item_attribute_data{Id := item_attribute.MagSize, Value:= 32.0}
            }
        },
    smg_data
        {
            ItemId:= item_smg_id.DualMicroSMGs_L4,
            DisplayName:= "Dual Micro Gs",
            Rarity := item_rarity.Epic,
            ItemGranterIndex := 9,
            Team := team_id.Blue,
            Attributes := array{         
                item_attribute_data{Id := item_attribute.DPS, Value:= 0.0},
                item_attribute_data{Id := item_attribute.Damage, Value:= 0.0},
                item_attribute_data{Id := item_attribute.MagSize, Value:= 0.0}
            }
        },
    smg_data
        {
            ItemId:= item_smg_id.DualMicroSMGs_L5,
            DisplayName:= "Dual Micro Gs",
            Rarity := item_rarity.Legendary,
            ItemGranterIndex := 10,
            Team := team_id.Blue,
            Attributes := array{         
                item_attribute_data{Id := item_attribute.DPS, Value:= 0.0},
                item_attribute_data{Id := item_attribute.Damage, Value:= 0.0},
                item_attribute_data{Id := item_attribute.MagSize, Value:= 0.0}
            }
        },
    smg_data
        {
            ItemId:= item_smg_id.HarbingerSMG_L1,
            DisplayName:= "Harbinger",
            Rarity := item_rarity.Common,
            ItemGranterIndex := 11,
            Team := team_id.Blue,
            Attributes := array{         
                item_attribute_data{Id := item_attribute.DPS, Value:= 0.0},
                item_attribute_data{Id := item_attribute.Damage, Value:= 0.0},
                item_attribute_data{Id := item_attribute.MagSize, Value:= 0.0}
            }
        },
    smg_data
        {
            ItemId:= item_smg_id.HarbingerSMG_L2,
            DisplayName:= "Harbinger",
            Rarity := item_rarity.Uncommon,
            ItemGranterIndex := 12,
            Team := team_id.Blue,
            Attributes := array{         
                item_attribute_data{Id := item_attribute.DPS, Value:= 0.0},
                item_attribute_data{Id := item_attribute.Damage, Value:= 0.0},
                item_attribute_data{Id := item_attribute.MagSize, Value:= 0.0}
            }
        },
    smg_data
        {
            ItemId:= item_smg_id.HarbingerSMG_L3,
            DisplayName:= "Harbinger",
            Rarity := item_rarity.Rare,
            ItemGranterIndex := 13,
            Team := team_id.Blue,
            Attributes := array{         
                item_attribute_data{Id := item_attribute.DPS, Value:= 190.0},
                item_attribute_data{Id := item_attribute.Damage, Value:= 19.0},
                item_attribute_data{Id := item_attribute.MagSize, Value:= 30.0}
            }
        },
    smg_data
        {
            ItemId:= item_smg_id.HarbingerSMG_L4,
            DisplayName:= "Harbinger",
            Rarity := item_rarity.Epic,
            ItemGranterIndex := 14,
            Team := team_id.Blue,
            Attributes := array{         
                item_attribute_data{Id := item_attribute.DPS, Value:= 0.0},
                item_attribute_data{Id := item_attribute.Damage, Value:= 0.0},
                item_attribute_data{Id := item_attribute.MagSize, Value:= 0.0}
            }
        },
    smg_data
        {
            ItemId:= item_smg_id.HarbingerSMG_L5,
            DisplayName:= "Harbinger",
            Rarity := item_rarity.Legendary,
            ItemGranterIndex := 15,
            Team := team_id.Blue,
            Attributes := array{         
                item_attribute_data{Id := item_attribute.DPS, Value:= 0.0},
                item_attribute_data{Id := item_attribute.Damage, Value:= 0.0},
                item_attribute_data{Id := item_attribute.MagSize, Value:= 0.0}
            }
        },
    smg_data
        {
            ItemId:= item_smg_id.RapidFireSMG_L1,
            DisplayName:= "Rapid Fire",
            Rarity := item_rarity.Common,
            ItemGranterIndex := 16,
            Team := team_id.Blue,
            Attributes := array{         
                item_attribute_data{Id := item_attribute.DPS, Value:= 0.0},
                item_attribute_data{Id := item_attribute.Damage, Value:= 0.0},
                item_attribute_data{Id := item_attribute.MagSize, Value:= 0.0}
            }
        },
    smg_data
        {
            ItemId:= item_smg_id.RapidFireSMG_L2,
            DisplayName:= "Rapid Fire",
            Rarity := item_rarity.Uncommon,
            ItemGranterIndex := 17,
            Team := team_id.Blue,
            Attributes := array{         
                item_attribute_data{Id := item_attribute.DPS, Value:= 0.0},
                item_attribute_data{Id := item_attribute.Damage, Value:= 0.0},
                item_attribute_data{Id := item_attribute.MagSize, Value:= 0.0}
            }
        },
    smg_data
        {
            ItemId:= item_smg_id.RapidFireSMG_L3,
            DisplayName:= "Rapid Fire",
            Rarity := item_rarity.Rare,
            ItemGranterIndex := 18,
            Team := team_id.Blue,
            Attributes := array{         
                item_attribute_data{Id := item_attribute.DPS, Value:= 225.0},
                item_attribute_data{Id := item_attribute.Damage, Value:= 15.0},
                item_attribute_data{Id := item_attribute.MagSize, Value:= 20.0}
            }
        },
    smg_data
        {
            ItemId:= item_smg_id.RapidFireSMG_L4,
            DisplayName:= "Rapid Fire",
            Rarity := item_rarity.Epic,
            ItemGranterIndex := 19,
            Team := team_id.Blue,
            Attributes := array{         
                item_attribute_data{Id := item_attribute.DPS, Value:= 0.0},
                item_attribute_data{Id := item_attribute.Damage, Value:= 0.0},
                item_attribute_data{Id := item_attribute.MagSize, Value:= 0.0}
            }
        },
    smg_data
        {
            ItemId:= item_smg_id.RapidFireSMG_L5,
            DisplayName:= "Rapid Fire",
            Rarity := item_rarity.Legendary,
            ItemGranterIndex := 20,
            Team := team_id.Blue,
            Attributes := array{         
                item_attribute_data{Id := item_attribute.DPS, Value:= 0.0},
                item_attribute_data{Id := item_attribute.Damage, Value:= 0.0},
                item_attribute_data{Id := item_attribute.MagSize, Value:= 0.0}
            }
        },
    smg_data
        {
            ItemId:= item_smg_id.RapidFireSMG_L6,
            DisplayName:= "Rapid Fire",
            Rarity := item_rarity.Mythic,
            ItemGranterIndex := 21,
            Team := team_id.Blue,
            Attributes := array{         
                item_attribute_data{Id := item_attribute.DPS, Value:= 0.0},
                item_attribute_data{Id := item_attribute.Damage, Value:= 0.0},
                item_attribute_data{Id := item_attribute.MagSize, Value:= 0.0}
            }
        },
    smg_data
        {
            ItemId:= item_smg_id.ThunderBurstSMG_L1,
            DisplayName:= "Thunder Burst",
            Rarity := item_rarity.Common,
            ItemGranterIndex := 22,
            Team := team_id.Blue,
            Attributes := array{         
                item_attribute_data{Id := item_attribute.DPS, Value:= 0.0},
                item_attribute_data{Id := item_attribute.Damage, Value:= 0.0},
                item_attribute_data{Id := item_attribute.MagSize, Value:= 0.0}
            }
        },
    smg_data
        {
            ItemId:= item_smg_id.ThunderBurstSMG_L2,
            DisplayName:= "Thunder Burst",
            Rarity := item_rarity.Uncommon,
            ItemGranterIndex := 23,
            Team := team_id.Blue,
            Attributes := array{         
                item_attribute_data{Id := item_attribute.DPS, Value:= 0.0},
                item_attribute_data{Id := item_attribute.Damage, Value:= 0.0},
                item_attribute_data{Id := item_attribute.MagSize, Value:= 0.0}
            }
        },
    smg_data
        {
            ItemId:= item_smg_id.ThunderBurstSMG_L3,
            DisplayName:= "Thunder Burst",
            Rarity := item_rarity.Rare,
            ItemGranterIndex := 24,
            Team := team_id.Blue,
            Attributes := array{         
                item_attribute_data{Id := item_attribute.DPS, Value:= 172.4},
                item_attribute_data{Id := item_attribute.Damage, Value:= 26.0},
                item_attribute_data{Id := item_attribute.MagSize, Value:= 27.0}
            }
        },
    smg_data
        {
            ItemId:= item_smg_id.ThunderBurstSMG_L4,
            DisplayName:= "Thunder Burst",
            Rarity := item_rarity.Epic,
            ItemGranterIndex := 25,
            Team := team_id.Blue,
            Attributes := array{         
                item_attribute_data{Id := item_attribute.DPS, Value:= 0.0},
                item_attribute_data{Id := item_attribute.Damage, Value:= 0.0},
                item_attribute_data{Id := item_attribute.MagSize, Value:= 0.0}
            }
        },
    smg_data
        {
            ItemId:= item_smg_id.ThunderBurstSMG_L5,
            DisplayName:= "Thunder Burst",
            Rarity := item_rarity.Legendary,
            ItemGranterIndex := 26,
            Team := team_id.Blue,
            Attributes := array{         
                item_attribute_data{Id := item_attribute.DPS, Value:= 0.0},
                item_attribute_data{Id := item_attribute.Damage, Value:= 0.0},
                item_attribute_data{Id := item_attribute.MagSize, Value:= 0.0}
            }
        },
    smg_data
        {
            ItemId:= item_smg_id.TwinMagSMG_L1,
            DisplayName:= "Twin Mag",
            Rarity := item_rarity.Common,
            ItemGranterIndex := 27,
            Team := team_id.Blue,
            Attributes := array{         
                item_attribute_data{Id := item_attribute.DPS, Value:= 0.0},
                item_attribute_data{Id := item_attribute.Damage, Value:= 0.0},
                item_attribute_data{Id := item_attribute.MagSize, Value:= 0.0}
            }
        },
    smg_data
        {
            ItemId:= item_smg_id.TwinMagSMG_L2,
            DisplayName:= "Twin Mag",
            Rarity := item_rarity.Uncommon,
            ItemGranterIndex := 28,
            Team := team_id.Blue,
            Attributes := array{         
                item_attribute_data{Id := item_attribute.DPS, Value:= 0.0},
                item_attribute_data{Id := item_attribute.Damage, Value:= 0.0},
                item_attribute_data{Id := item_attribute.MagSize, Value:= 0.0}
            }
        },
    smg_data
        {
            ItemId:= item_smg_id.TwinMagSMG_L3,
            DisplayName:= "Twin Mag",
            Rarity := item_rarity.Rare,
            ItemGranterIndex := 29,
            Team := team_id.Blue,
            Attributes := array{         
                item_attribute_data{Id := item_attribute.DPS, Value:= 234.0},
                item_attribute_data{Id := item_attribute.Damage, Value:= 18.0},
                item_attribute_data{Id := item_attribute.MagSize, Value:= 20.0}
            }
        },
    smg_data
        {
            ItemId:= item_smg_id.TwinMagSMG_L4,
            DisplayName:= "Twin Mag",
            Rarity := item_rarity.Epic,
            ItemGranterIndex := 30,
            Team := team_id.Blue,
            Attributes := array{         
                item_attribute_data{Id := item_attribute.DPS, Value:= 0.0},
                item_attribute_data{Id := item_attribute.Damage, Value:= 0.0},
                item_attribute_data{Id := item_attribute.MagSize, Value:= 0.0}
            }
        },
    smg_data
        {
            ItemId:= item_smg_id.TwinMagSMG_L5,
            DisplayName:= "Twin Mag",
            Rarity := item_rarity.Legendary,
            ItemGranterIndex := 31,
            Team := team_id.Blue,
            Attributes := array{         
                item_attribute_data{Id := item_attribute.DPS, Value:= 0.0},
                item_attribute_data{Id := item_attribute.Damage, Value:= 0.0},
                item_attribute_data{Id := item_attribute.MagSize, Value:= 0.0}
            }
        }
    }