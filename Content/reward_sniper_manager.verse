using { /Fortnite.com/Devices }
using { /Verse.org/Random }
using { /Verse.org/Assets }
using { /Verse.org/Simulation }
using { /UnrealEngine.com/Temporary/Diagnostics }

log_reward_sniper_manager := class(log_channel){}

reward_sniper_manager := class(creative_device,grantable):

    Logger:log = log{Channel:=log_reward_sniper_manager}

    @editable
    SniperDefinition<private>:reward_sniper_definition = reward_sniper_definition{}   

    FindItemByRarity<override>(ItemsData:[]item_data, Rarity:item_rarity):item_data =
        Logger.Print("Sniper Manager")
        var ItemsByRarity:[]item_data = array{}
        var ItemData:item_data = item_data{}  
        for (TempItemData:ItemsData):
            if (TempItemData.Rarity = Rarity):
                set ItemsByRarity += array{TempItemData} 
        if:
            ItemsByRarity.Length > 0            
        then:
            set ItemData = GetRandomItem(ItemsByRarity)       
        return ItemData 

    FindItemsByTeam<override>(ItemsData:[]item_data, Team:team_id):[]item_data =
        var ItemsByTeam:[]item_data = array{}
        for (ItemData:ItemsData):
            if (ItemData.Team = Team):
                set ItemsByTeam += array{ItemData} 
        if:
            ItemsByTeam.Length <= 0            
        then:
            Logger.Print("There are not elements") 
        return ItemsByTeam  

    GetRandomItem<private>(ItemsData:[]item_data): item_data=
        var NewItem: item_data = item_data{}
        IndexItemRandon := GetRandomInt(0, ItemsData.Length-1)
        if:
            ItemChoosen := ItemsData[IndexItemRandon]
            SniperData := sniper_data[ItemChoosen]
        then:
            set NewItem = SniperData
        return NewItem	    

    GrantItem<override>(Agent:agent, Index:int):void= 
        Logger.Print("Sniper Granter")
        SniperDefinition.GrantItemToPlayer(Agent, Index)

    GetItemsData<override>():[]item_data =
        SniperDefinition.GetSniperData()

    HasRarity<override>(Rarity:item_rarity): logic=
        ItemsData := GetItemsData()
        for (ItemData:ItemsData):
            if (ItemData.Rarity = Rarity):
                return true
        return false

    HasRarity<override>(ItemsData:[]item_data,Rarity:item_rarity): logic=        
        for (ItemData:ItemsData):
            if (ItemData.Rarity = Rarity):
                return true
        return false
        
sniper_data := class<concrete>(weapon_data,iconTextureable):
    @editable
    ItemId: item_sniper_id = item_sniper_id.Undefined 
    
    WeaponType<override>:weapon = weapon.Sniper
    
    GetIconTexture<override>()<computes>: texture = FindSniperIconTexture(ItemId)

reward_sniper_definition := class<concrete>:
    @editable # One item granter with all the Snipers
    SniperGranter<private>:item_granter_device = item_granter_device{}	

    GrantItemToPlayer<public>(Agent:agent, Index:int):void= 
        SniperGranter.GrantItemIndex(Agent, Index)
        
    GetSniperData<public>():[]item_data =
        return SniperData

    @editable
    SniperData<private>:[]sniper_data := array{	
        sniper_data
        {
            ItemId:= item_sniper_id.AutomaticSniperRifle_L2,
            DisplayName:= "Automatic Sniper Rifle",
            Rarity := item_rarity.Uncommon,
            ItemGranterIndex := 0,
            Team := team_id.Blue,
            Attributes := array{         
                item_attribute_data{Id := item_attribute.DPS, Value:= 0.0},
            item_attribute_data{Id := item_attribute.Damage, Value:= 0.0},
            item_attribute_data{Id := item_attribute.MagSize, Value:= 0.0}
            }
        },
        sniper_data
        {
            ItemId:= item_sniper_id.AutomaticSniperRifle_L3,
            DisplayName:= "Automatic Sniper Rifle",
            Rarity := item_rarity.Rare,
            ItemGranterIndex := 1,
            Team := team_id.Blue,
            Attributes := array{         
                item_attribute_data{Id := item_attribute.DPS, Value:= 160.0},
            item_attribute_data{Id := item_attribute.Damage, Value:= 40.0},
            item_attribute_data{Id := item_attribute.MagSize, Value:= 80.0}
            }
        },
        sniper_data
        {
            ItemId:= item_sniper_id.AutomaticSniperRifle_L4,
            DisplayName:= "Automatic Sniper Rifle",
            Rarity := item_rarity.Epic,
            ItemGranterIndex := 2,
            Team := team_id.Blue,
            Attributes := array{         
                item_attribute_data{Id := item_attribute.DPS, Value:= 0.0},
            item_attribute_data{Id := item_attribute.Damage, Value:= 0.0},
            item_attribute_data{Id := item_attribute.MagSize, Value:= 0.0}
            }
        },
        sniper_data
        {
            ItemId:= item_sniper_id.AutomaticSniperRifle_L5,
            DisplayName:= "Automatic Sniper Rifle",
            Rarity := item_rarity.Legendary,
            ItemGranterIndex := 3,
            Team := team_id.Blue,
            Attributes := array{         
                item_attribute_data{Id := item_attribute.DPS, Value:= 0.0},
            item_attribute_data{Id := item_attribute.Damage, Value:= 0.0},
            item_attribute_data{Id := item_attribute.MagSize, Value:= 0.0}
            }
        },
        sniper_data
        {
            ItemId:= item_sniper_id.BoltActionSniperRifle_L1,
            DisplayName:= "Bolt Action Sniper Rifle",
            Rarity := item_rarity.Common,
            ItemGranterIndex := 4,
            Team := team_id.Blue,
            Attributes := array{         
                item_attribute_data{Id := item_attribute.DPS, Value:= 0.0},
            item_attribute_data{Id := item_attribute.Damage, Value:= 0.0},
            item_attribute_data{Id := item_attribute.MagSize, Value:= 0.0}
            }
        },
        sniper_data
        {
            ItemId:= item_sniper_id.BoltActionSniperRifle_L2,
            DisplayName:= "Bolt Action Sniper Rifle",
            Rarity := item_rarity.Uncommon,
            ItemGranterIndex := 5,
            Team := team_id.Blue,
            Attributes := array{         
                item_attribute_data{Id := item_attribute.DPS, Value:= 0.0},
            item_attribute_data{Id := item_attribute.Damage, Value:= 0.0},
            item_attribute_data{Id := item_attribute.MagSize, Value:= 0.0}
            }
        },
        sniper_data
        {
            ItemId:= item_sniper_id.BoltActionSniperRifle_L3,
            DisplayName:= "Bolt Action Sniper Rifle",
            Rarity := item_rarity.Rare,
            ItemGranterIndex := 6,
            Team := team_id.Blue,
            Attributes := array{         
                item_attribute_data{Id := item_attribute.DPS, Value:= 82.5},
            item_attribute_data{Id := item_attribute.Damage, Value:= 110.0},
            item_attribute_data{Id := item_attribute.MagSize, Value:= 275.0}
            }
        },
        sniper_data
        {
            ItemId:= item_sniper_id.BoltActionSniperRifle_L4,
            DisplayName:= "Bolt Action Sniper Rifle",
            Rarity := item_rarity.Epic,
            ItemGranterIndex := 7,
            Team := team_id.Blue,
            Attributes := array{         
                item_attribute_data{Id := item_attribute.DPS, Value:= 0.0},
            item_attribute_data{Id := item_attribute.Damage, Value:= 0.0},
            item_attribute_data{Id := item_attribute.MagSize, Value:= 0.0}
            }
        },
        sniper_data
        {
            ItemId:= item_sniper_id.BoltActionSniperRifle_L5,
            DisplayName:= "Bolt Action Sniper Rifle",
            Rarity := item_rarity.Legendary,
            ItemGranterIndex := 8,
            Team := team_id.Blue,
            Attributes := array{         
                item_attribute_data{Id := item_attribute.DPS, Value:= 0.0},
            item_attribute_data{Id := item_attribute.Damage, Value:= 0.0},
            item_attribute_data{Id := item_attribute.MagSize, Value:= 0.0}
            }
        },
        sniper_data
        {
            ItemId:= item_sniper_id.CobraDMR_L1,
            DisplayName:= "Cobra",
            Rarity := item_rarity.Common,
            ItemGranterIndex := 9,
            Team := team_id.Blue,
            Attributes := array{         
                item_attribute_data{Id := item_attribute.DPS, Value:= 0.0},
            item_attribute_data{Id := item_attribute.Damage, Value:= 0.0},
            item_attribute_data{Id := item_attribute.MagSize, Value:= 0.0}
            }
        },
        sniper_data
        {
            ItemId:= item_sniper_id.CobraDMR_L2,
            DisplayName:= "Cobra",
            Rarity := item_rarity.Uncommon,
            ItemGranterIndex := 10,
            Team := team_id.Blue,
            Attributes := array{         
                item_attribute_data{Id := item_attribute.DPS, Value:= 0.0},
            item_attribute_data{Id := item_attribute.Damage, Value:= 0.0},
            item_attribute_data{Id := item_attribute.MagSize, Value:= 0.0}
            }
        },
        sniper_data
        {
            ItemId:= item_sniper_id.CobraDMR_L3,
            DisplayName:= "Cobra",
            Rarity := item_rarity.Rare,
            ItemGranterIndex := 11,
            Team := team_id.Blue,
            Attributes := array{         
                item_attribute_data{Id := item_attribute.DPS, Value:= 160.0},
            item_attribute_data{Id := item_attribute.Damage, Value:= 40.0},
            item_attribute_data{Id := item_attribute.MagSize, Value:= 66.0}
            }
        },
        sniper_data
        {
            ItemId:= item_sniper_id.CobraDMR_L4,
            DisplayName:= "Cobra",
            Rarity := item_rarity.Epic,
            ItemGranterIndex := 12,
            Team := team_id.Blue,
            Attributes := array{         
                item_attribute_data{Id := item_attribute.DPS, Value:= 0.0},
            item_attribute_data{Id := item_attribute.Damage, Value:= 0.0},
            item_attribute_data{Id := item_attribute.MagSize, Value:= 0.0}
            }
        },
        sniper_data
        {
            ItemId:= item_sniper_id.CobraDMR_L5,
            DisplayName:= "Cobra",
            Rarity := item_rarity.Legendary,
            ItemGranterIndex := 13,
            Team := team_id.Blue,
            Attributes := array{         
                item_attribute_data{Id := item_attribute.DPS, Value:= 0.0},
            item_attribute_data{Id := item_attribute.Damage, Value:= 0.0},
            item_attribute_data{Id := item_attribute.MagSize, Value:= 0.0}
            }
        },
        sniper_data
        {
            ItemId:= item_sniper_id.CobraDMR_L6,
            DisplayName:= "Cobra",
            Rarity := item_rarity.Mythic,
            ItemGranterIndex := 14,
            Team := team_id.Blue,
            Attributes := array{         
                item_attribute_data{Id := item_attribute.DPS, Value:= 0.0},
            item_attribute_data{Id := item_attribute.Damage, Value:= 0.0},
            item_attribute_data{Id := item_attribute.MagSize, Value:= 0.0}
            }
        },
        sniper_data
        {
            ItemId:= item_sniper_id.ExCaliberRifle_L1,
            DisplayName:= "Ex Caliber Rifle",
            Rarity := item_rarity.Common,
            ItemGranterIndex := 15,
            Team := team_id.Blue,
            Attributes := array{         
                item_attribute_data{Id := item_attribute.DPS, Value:= 0.0},
            item_attribute_data{Id := item_attribute.Damage, Value:= 0.0},
            item_attribute_data{Id := item_attribute.MagSize, Value:= 0.0}
            }
        },
        sniper_data
        {
            ItemId:= item_sniper_id.ExCaliberRifle_L2,
            DisplayName:= "Ex Caliber Rifle",
            Rarity := item_rarity.Uncommon,
            ItemGranterIndex := 16,
            Team := team_id.Blue,
            Attributes := array{         
                item_attribute_data{Id := item_attribute.DPS, Value:= 0.0},
            item_attribute_data{Id := item_attribute.Damage, Value:= 0.0},
            item_attribute_data{Id := item_attribute.MagSize, Value:= 0.0}
            }
        },
        sniper_data
        {
            ItemId:= item_sniper_id.ExCaliberRifle_L3,
            DisplayName:= "Ex Caliber Rifle",
            Rarity := item_rarity.Rare,
            ItemGranterIndex := 17,
            Team := team_id.Blue,
            Attributes := array{         
                item_attribute_data{Id := item_attribute.DPS, Value:= 0.0},
            item_attribute_data{Id := item_attribute.Damage, Value:= 45.0},
            item_attribute_data{Id := item_attribute.MagSize, Value:= 113.0}
            }
        },
        sniper_data
        {
            ItemId:= item_sniper_id.ExCaliberRifle_L4,
            DisplayName:= "Ex Caliber Rifle",
            Rarity := item_rarity.Epic,
            ItemGranterIndex := 18,
            Team := team_id.Blue,
            Attributes := array{         
                item_attribute_data{Id := item_attribute.DPS, Value:= 0.0},
            item_attribute_data{Id := item_attribute.Damage, Value:= 0.0},
            item_attribute_data{Id := item_attribute.MagSize, Value:= 0.0}
            }
        },
        sniper_data
        {
            ItemId:= item_sniper_id.ExCaliberRifle_L5,
            DisplayName:= "Ex Caliber Rifle",
            Rarity := item_rarity.Legendary,
            ItemGranterIndex := 19,
            Team := team_id.Blue,
            Attributes := array{         
                item_attribute_data{Id := item_attribute.DPS, Value:= 0.0},
            item_attribute_data{Id := item_attribute.Damage, Value:= 0.0},
            item_attribute_data{Id := item_attribute.MagSize, Value:= 0.0}
            }
        },
        sniper_data
        {
            ItemId:= item_sniper_id.HuntressDMR_L2,
            DisplayName:= "Huntress",
            Rarity := item_rarity.Uncommon,
            ItemGranterIndex := 20,
            Team := team_id.Blue,
            Attributes := array{         
                item_attribute_data{Id := item_attribute.DPS, Value:= 0.0},
            item_attribute_data{Id := item_attribute.Damage, Value:= 0.0},
            item_attribute_data{Id := item_attribute.MagSize, Value:= 0.0}
            }
        },
        sniper_data
        {
            ItemId:= item_sniper_id.HuntressDMR_L3,
            DisplayName:= "Huntress",
            Rarity := item_rarity.Rare,
            ItemGranterIndex := 21,
            Team := team_id.Blue,
            Attributes := array{         
                item_attribute_data{Id := item_attribute.DPS, Value:= 94.9},
            item_attribute_data{Id := item_attribute.Damage, Value:= 65.0},
            item_attribute_data{Id := item_attribute.MagSize, Value:= 113.8}
            }
        },
        sniper_data
        {
            ItemId:= item_sniper_id.HuntressDMR_L4,
            DisplayName:= "Huntress",
            Rarity := item_rarity.Epic,
            ItemGranterIndex := 22,
            Team := team_id.Blue,
            Attributes := array{         
                item_attribute_data{Id := item_attribute.DPS, Value:= 0.0},
            item_attribute_data{Id := item_attribute.Damage, Value:= 0.0},
            item_attribute_data{Id := item_attribute.MagSize, Value:= 0.0}
            }
        },
            sniper_data
        {
            ItemId:= item_sniper_id.HuntressDMR_L5,
            DisplayName:= "Huntress",
            Rarity := item_rarity.Legendary,
            ItemGranterIndex := 23,
            Team := team_id.Blue,
            Attributes := array{         
                item_attribute_data{Id := item_attribute.DPS, Value:= 0.0},
            item_attribute_data{Id := item_attribute.Damage, Value:= 0.0},
            item_attribute_data{Id := item_attribute.MagSize, Value:= 0.0}
            }
        },
        sniper_data
        {
            ItemId:= item_sniper_id.ReaperSniperRifle_L2,
            DisplayName:= "Reaper Sniper Rifle",
            Rarity := item_rarity.Uncommon,
            ItemGranterIndex := 24,
            Team := team_id.Blue,
            Attributes := array{         
                item_attribute_data{Id := item_attribute.DPS, Value:= 0.0},
            item_attribute_data{Id := item_attribute.Damage, Value:= 0.0},
            item_attribute_data{Id := item_attribute.MagSize, Value:= 0.0}
            }
        },
        sniper_data
        {
            ItemId:= item_sniper_id.ReaperSniperRifle_L3,
            DisplayName:= "Reaper Sniper Rifle",
            Rarity := item_rarity.Rare,
            ItemGranterIndex := 25,
            Team := team_id.Blue,
            Attributes := array{         
                item_attribute_data{Id := item_attribute.DPS, Value:= 58.3},
            item_attribute_data{Id := item_attribute.Damage, Value:= 110.0},
            item_attribute_data{Id := item_attribute.MagSize, Value:= 275.0}
            }
        },
        sniper_data
        {
            ItemId:= item_sniper_id.ReaperSniperRifle_L4,
            DisplayName:= "Reaper Sniper Rifle",
            Rarity := item_rarity.Epic,
            ItemGranterIndex := 26,
            Team := team_id.Blue,
            Attributes := array{         
                item_attribute_data{Id := item_attribute.DPS, Value:= 0.0},
            item_attribute_data{Id := item_attribute.Damage, Value:= 0.0},
            item_attribute_data{Id := item_attribute.MagSize, Value:= 0.0}
            }
        },
        sniper_data
        {
            ItemId:= item_sniper_id.ReaperSniperRifle_L5,
            DisplayName:= "Reaper Sniper Rifle",
            Rarity := item_rarity.Legendary,
            ItemGranterIndex := 27,
            Team := team_id.Blue,
            Attributes := array{         
                item_attribute_data{Id := item_attribute.DPS, Value:= 0.0},
            item_attribute_data{Id := item_attribute.Damage, Value:= 0.0},
            item_attribute_data{Id := item_attribute.MagSize, Value:= 0.0}
            }
        }             
    }