using { /Fortnite.com/Devices }
using { /Verse.org/Simulation }
using { /UnrealEngine.com/Temporary/Diagnostics }

grantable := interface():
    FindItemByRarity(ItemsData:[]item_data, Rarity:item_rarity):item_data   
    FindItemsByTeam(ItemsData:[]item_data, Team:team_id):[]item_data 
    GrantItem(Agent:agent, Index:int):void
    GetItemsData():[]item_data  
    HasRarity(Rarity:item_rarity): logic
    HasRarity(ItemsData:[]item_data,Rarity:item_rarity): logic 

weapon_data := class(item_data, attributable):
    WeaponType:weapon = weapon.Rifle
    GetMaybeAttributes<override>()<computes>:?[]item_attribute_data = option{Attributes}

reward_weapon_definition := class(creative_device):

    @editable
    RifleManager:reward_rifle_manager = reward_rifle_manager{}
    
    @editable
    ShotgunManager:reward_shotgun_manager = reward_shotgun_manager{}

    @editable
    SMGManager:reward_smg_manager = reward_smg_manager{}
    
    @editable
    PistolManager:reward_pistol_manager = reward_pistol_manager{}

    @editable
    DMRManager:reward_dmr_manager = reward_dmr_manager{}

    @editable
    SniperManager:reward_sniper_manager = reward_sniper_manager{}

    @editable
    ExplosiveManager:reward_explosive_manager = reward_explosive_manager{}

    @editable
    OthersManager:reward_others_manager = reward_others_manager{}

    GetGrantableFirearm(Id:weapon): grantable=
        var MaybeGrantable:?grantable = false
        case (Id):  
            weapon.Rifle  => set MaybeGrantable = option{RifleManager}
            weapon.Shotgun => set MaybeGrantable = option{ShotgunManager}
            weapon.Smg => set MaybeGrantable = option{SMGManager}
            weapon.Pistol => set MaybeGrantable = option{PistolManager}
            weapon.Dmr => set MaybeGrantable = option{DMRManager}
            weapon.Sniper => set MaybeGrantable = option{SniperManager}
            weapon.Explosive => set MaybeGrantable = option{ExplosiveManager}
            weapon.Others => set MaybeGrantable = option{OthersManager}
        if:
            Grantable := MaybeGrantable?
        then:
            return Grantable
        return reward_rifle_manager{}

weapon := enum {Rifle, Shotgun, Smg, Pistol, Dmr, Sniper, Explosive, Others}

# ADD to this array the available firearm type
AllWeaponsArray:[]weapon = array:
    weapon.Rifle,
    weapon.Shotgun,
    weapon.Smg,
    weapon.Pistol,
    weapon.Dmr,
    weapon.Sniper,
    weapon.Explosive,
    weapon.Others
    # ...

ToString(Id:weapon): string =
    case(Id):
        weapon.Rifle => "Rifle"
        weapon.Shotgun => "Shotgun"
        weapon.Smg => "Smg"	
        weapon.Pistol => "Pistol"	      
        weapon.Dmr => "Dmr"	    
        weapon.Sniper => "Sniper"	
        weapon.Explosive => "Explosive"	     
        weapon.Others => "Others"	     

ToWeapons(Index:int) : weapon =
    case(Index):
        0 => weapon.Rifle 
        1 => weapon.Shotgun 
        2 => weapon.Smg 
        3 => weapon.Pistol 
        4 => weapon.Dmr
        5 => weapon.Sniper
        6 => weapon.Explosive
        7 => weapon.Others
        _ => weapon.Rifle 

GetweaponIndex(Id:weapon) : int =
    case(Id):
        weapon.Rifle => 0
        weapon.Shotgun => 1
        weapon.Smg => 2
        weapon.Pistol => 3
        weapon.Dmr => 4
        weapon.Sniper => 5
        weapon.Explosive => 6
        weapon.Others => 7