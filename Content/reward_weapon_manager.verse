using { /Fortnite.com/Devices }
using { /Verse.org/Random }
using { /Verse.org/Assets }
using { /Verse.org/Simulation }
using { /UnrealEngine.com/Temporary/Diagnostics }

log_reward_weapon_manager := class(log_channel){}

<# 
Class that implements the strategy design pattern.
#>
item_handler_context := class:

    var MaybeStrategy:?grantable = false 

    SetStrategy(NewStrategy:grantable):void=
        set MaybeStrategy = option{NewStrategy}

    FindItemByRarity(ItemsData:[]item_data, Rarity:item_rarity):item_data =
        var ItemData:item_data = item_data{} 
        if:
            Strategy := MaybeStrategy?            
        then:
            set ItemData = Strategy.FindItemByRarity(ItemsData, Rarity)
        return ItemData

    FindItemsByTeam(ItemsData:[]item_data, Team:team_id):[]item_data =
        var ItemsByTeam:[]item_data = array{}
        if:
            Strategy := MaybeStrategy?            
        then:
            set ItemsByTeam = Strategy.FindItemsByTeam(ItemsData, Team)
        return ItemsByTeam

    GrantItem(Agent:agent, Index:int):void =
        if:
            Strategy := MaybeStrategy?            
        then:
            Strategy.GrantItem(Agent, Index) 
            
    GetItemsData():[]item_data=
        var ItemsData:[]item_data = array{}
        if:
            Strategy := MaybeStrategy?    
        then:            
            set ItemsData = Strategy.GetItemsData() 
        return ItemsData  

    HasRarity(Rarity:item_rarity): logic=
        if:
            Strategy := MaybeStrategy?    
        then:            
            return Strategy.HasRarity(Rarity)
        return false

    HasRarity(ItemsData:[]item_data,Rarity:item_rarity): logic=
        if:
            Strategy := MaybeStrategy?    
        then:            
            return Strategy.HasRarity(ItemsData,Rarity)
        return false


reward_weapon_manager := class(creative_device,awardable):
    Logger:log = log{Channel:=log_reward_weapon_manager}    

    @editable
    WeaponDefinitions:reward_weapon_definition = reward_weapon_definition{}

    GetRewardableData<override>(RewardableData:rewardable_data): rewardable =
        var NewRarity:item_rarity = RewardableData.Rarity
        var MaxIterations: int = 5
        var CurrentIteration: int = 0 
        RandomNumber := GetRandomInt(0, AllWeaponsArray.Length - 1)
        if:
            IdFirearm := AllWeaponsArray[RandomNumber]
        then:
            ItemHandler := item_handler_context{}
            SetStrategy(ItemHandler,IdFirearm)
            loop:
                bHasRarity := ItemHandler.HasRarity(NewRarity)
                if:
                    bHasRarity?
                then:
                    break      
                if:
                    CurrentIteration >= MaxIterations 
                then:                
                    set NewRarity = item_rarity.Common 
                    break          
                set NewRarity = HandleRarity(NewRarity) 
                set CurrentIteration += 1
            return GetRandomRewardableByRarity(ItemHandler,rewardable_data{Team:=RewardableData.Team, TypeSearch:=RewardableData.TypeSearch,Rarity:=NewRarity})
        return item_data{}         

    GetRandomRewardableByRarity(ItemHandler:item_handler_context,RewardableData:rewardable_data): rewardable =      
        if:
            RewardableData.TypeSearch = type_search.All
        then:
            return NewItem := GetItemData(GetItemsData(ItemHandler),RewardableData.Rarity,ItemHandler)
        else:
            AllItemsByTeam := ItemHandler.FindItemsByTeam(GetItemsData(ItemHandler), RewardableData.Team)
            bHasTeam :=  ItemHandler.HasRarity(AllItemsByTeam,RewardableData.Rarity)
            if:
                bHasTeam?
                AllItemsByTeam.Length > 0
            then:
                return NewItem := GetItemData(AllItemsByTeam,RewardableData.Rarity,ItemHandler)    
            else:
                return NewItem := GetItemData(GetItemsData(ItemHandler),RewardableData.Rarity,ItemHandler)        
        
    GetItemsData(ItemHandler:item_handler_context):[]item_data=
        ItemHandler.GetItemsData()

    GetItemData(ItemsData:[]item_data, Rarity:item_rarity,ItemHandler:item_handler_context):item_data=
        ItemData := ItemHandler.FindItemByRarity(ItemsData, Rarity)
        return ItemData    

    GiveReward<override>(Player:player, Rewardable:rewardable): void=        
        if:
            Item := item_data[Rewardable]  
            WeaponData := weapon_data[Item]
            FirearmType := WeaponData.WeaponType            
            ItemGranterIndex := Item.ItemGranterIndex           
            Agent := agent[Player]            
        then:                
            ItemHandler := item_handler_context{}
            SetStrategy(ItemHandler,FirearmType)
            GrantItem(Agent, ItemGranterIndex,ItemHandler)
            
    GrantItem(Agent:agent, ItemGranterIndex:int,ItemHandler:item_handler_context): void =
        ItemHandler.GrantItem(Agent, ItemGranterIndex)   

    SetStrategy(ItemHandler:item_handler_context,Id:weapon):void=
        case (Id):           
            weapon.Rifle  => ItemHandler.SetStrategy(WeaponDefinitions.GetGrantableFirearm(Id))
            weapon.Shotgun => ItemHandler.SetStrategy(WeaponDefinitions.GetGrantableFirearm(Id))
            weapon.Smg => ItemHandler.SetStrategy(WeaponDefinitions.GetGrantableFirearm(Id))
            weapon.Pistol => ItemHandler.SetStrategy(WeaponDefinitions.GetGrantableFirearm(Id))
            weapon.Dmr => ItemHandler.SetStrategy(WeaponDefinitions.GetGrantableFirearm(Id))
            weapon.Sniper => ItemHandler.SetStrategy(WeaponDefinitions.GetGrantableFirearm(Id))
            weapon.Explosive => ItemHandler.SetStrategy(WeaponDefinitions.GetGrantableFirearm(Id)) 
            weapon.Others => ItemHandler.SetStrategy(WeaponDefinitions.GetGrantableFirearm(Id)) 