using { /Fortnite.com/Devices }
using { /Fortnite.com/Characters }
using { /Verse.org/Simulation }
using { /UnrealEngine.com/Temporary/Diagnostics }
using { /UnrealEngine.com/Temporary/SpatialMath }

log_roguelike_game_mode := class(log_channel){}

<# 
Main Roguelike game mode class
#>
roguelike_game_mode := class(creative_device, stat_sender):
    Logger:log = log{Channel:=log_roguelike_game_mode}

    @editable # Team settings and inventory for the players team
    PlayerTeamManager:team_settings_and_inventory_device = team_settings_and_inventory_device{}

    @editable
    RoomManager:room_manager = room_manager{}

    @editable # Timer to show how much time has taken the player to get to the goal
    GoalTimerDevice:timer_device = timer_device{}

    @editable
    GoalCompletedHUDMessage:hud_message_device = hud_message_device{}

    @editable
    GoalCompletedAccolade:accolades_device = accolades_device{}

    @editable
    CompleteGoalString:string = "You have escaped in "

    @editable
    CompleteGoalString2:string = "Can you do it faster?"

    @editable
    ReplayButton:button_device = button_device{}

    @editable
    ReplayTeleporter:teleporter_device = teleporter_device{}

    @editable
    ReplayAnalytics:analytics_device = analytics_device{}

    @editable
    ReplayAccolade:accolades_device = accolades_device{}

    @editable
    PlayerControllerManager:player_controller_manager = player_controller_manager{}

    @editable
    PlayerPersistableData:player_persistable_data = player_persistable_data{}

    @editable
    HudManager:hud_manager = hud_manager{}

    @editable # If true, the HUD will only be initialized once the specified cinematic stops
    bHudWaitsForCinematic : logic = true

    @editable # If WaitForCinematic is true, the player hud will only be initialized after this cinematic stops
    HudCinematicSequenceTrigger : cinematic_sequence_device = cinematic_sequence_device{}

    @editable # If something goes wrong with the cinematic, show the HUD after this time has passed!
    CinematicFailsafeTime : float = 30.0

    @editable
    CurrencyManager:clash_currency_manager = clash_currency_manager{}

    @editable
    DBNODevice:down_but_not_out_device = down_but_not_out_device{}

    @editable
    RevivalDistance:float = 0.0

    var StatsManager:?stats_manager = false

    OnBegin<override>()<suspends>:void =
        Sleep(1.0) # Start after we are sure other manager has been initialized correctly
        Logger.Print("Round loaded")
        
        if(FoundStatsManager := GetStatManager[]):
            set StatsManager = option{FoundStatsManager}

        ReplayButton.InteractedWithEvent.Subscribe(OnReplayButtonInteracted)

        Players:[]player = GetPlayspace().GetPlayers()
        for (Player:Players):
            InitializePlayer(Player, false)
        
        GetPlayspace().PlayerAddedEvent().Subscribe(OnPlayerAdded)
        GetPlayspace().PlayerRemovedEvent().Subscribe(OnPlayerRemoved)

        PlayerTeamManager.TeamMemberEliminatedEvent.Subscribe(OnAgentEliminated)
        PlayerTeamManager.TeamMemberSpawnedEvent.Subscribe(OnAgentSpawned)

        DBNODevice.AgentRevivedEvent.Subscribe(OnAgentRevived)
        
        AwaitPlayersArriveGoal()

    InitializePlayer(Player:player, bIsJoinInProgress:logic):void =
        PlayerControllerManager.RegisterPlayer(Player) 
        RoomManager.RegisterPlayer(Player)
        
        # First initialize persistent data, then notify the managers that depend on it
        if (PlayerState := PlayerPersistableData.GetOrAddPlayerState[Player]):
            PlayerPersistableData.InitializePlayerState(Player, PlayerState, CurrencyManager)

        # HUD after initializing all
        spawn { SetupHudForPlayer(Player, bIsJoinInProgress) }

    SetupHudForPlayer(Player:player, bIsJoinInProgress:logic)<suspends>:void=
        if (bHudWaitsForCinematic? and not bIsJoinInProgress?): # Players that join in progress do not see the cinematic
            race:
                HudCinematicSequenceTrigger.StoppedEvent.Await()
                Sleep(CinematicFailsafeTime)
        else:
            Sleep(1.0) # We wait 1 sec before initializing HUD to prevent an error when it was initialized and updates were not displayed to player

        HudManager.OnPlayerAdded(Player)

    OnPlayerAdded(Player:player):void =
        Logger.Print("A player joined the game")
        
        InitializePlayer(Player, true)

    OnPlayerRemoved(Player:player):void =
        Logger.Print("A player left the game")

        PlayerControllerManager.UnregisterPlayer(Player)
        RoomManager.UnregisterPlayer(Player)
        HudManager.OnPlayerRemoved(Player)

    OnAgentEliminated(Agent:agent):void =
        if (Player:=player[Agent]):
            PlayerControllerManager.OnPlayerEliminated(Player)
            RoomManager.OnPlayerEliminated(Player)

    OnAgentSpawned(Agent:agent):void =
        if (Player:=player[Agent]):
            PlayerControllerManager.OnPlayerSpawned(Player)

    AwaitPlayersArriveGoal()<suspends>:void =
        sync:
            loop:
                Player:player = RoomManager.OnPlayerEnteredRoomBeforeGoal.Await()
                # Run effects must be removed
                PlayerControllerManager.OnPlayerEnteredRoomBeforeGoal(Player)

            loop:
                Player:player = RoomManager.OnGoalStarted.Await()
                Logger.Print("Player started goal")
                GoalTimerDevice.Reset(Player)
                GoalTimerDevice.Start(Player)
                
            #     # Verify if player in UI and if so pause timer
            #     if (PlayerControllerManager.GetPlayerController[Player].IsInState[player_stat.Invulnerable ]):
            #         GoalTimerDevice.Pause(Player)

            # loop:
            #     StatChangedData := PlayerControllerManager.PlayerStatChangedEvent.Await()

            #     # Pause the timer when the player is in UI # TODO probably we need to use a UI variable for this instead of vulnerability
            #     if (StatChangedData.Stat = player_stat.Invulnerable):
            #         if:
            #             Player:=StatChangedData.Player
            #             PlayerRoomData:=RoomManager.GetPlayerRoomData[Player]
            #             PlayerRoomData.bHasStartedGoal? and not PlayerRoomData.bHasCompletedGoal?
            #         then:
            #             if (StatChangedData.NewValue > 0.0):
            #                 GoalTimerDevice.Pause(Player)
            #             else:
            #                 GoalTimerDevice.Resume(Player)

            loop:
                Player:player = RoomManager.OnGoalCancelled.Await()
                Logger.Print("Player cancelled goal")
                if (GoalDuration:= Floor[GoalTimerDevice.GetMaxDuration() - GoalTimerDevice.GetActiveDuration(Player)]):
                    Data := array {
                        stat_data{Player := Player, Category := stat_category.Run, Type := stat_type_new.Number, ID := stat_id_new.Playtime, Value := GoalDuration*1.0}
                        }
                    if(SendStats[Data]):
                GoalTimerDevice.Reset(Player)

            loop:
                Player:player = RoomManager.OnGoalCompleted.Await()
                Logger.Print("Player completed goal")
                
                if (GoalDuration:= Floor[GoalTimerDevice.GetMaxDuration() - GoalTimerDevice.GetActiveDuration(Player)]):
                    GoalCompletedHUDMessage.Show(Player, StringToMessage(GetCompleteGoalMessage(GoalDuration)))
                    Data := array {
                    stat_data{Player := Player, Category := stat_category.General, Type := stat_type_new.Number, ID := stat_id_new.Escapes, Value := 1.0}
                    stat_data{Player := Player, Category := stat_category.General, Type := stat_type_new.Number, ID := stat_id_new.EscapeStreak, Value := 1.0}
                    stat_data{Player := Player, Category := stat_category.General, Type := stat_type_new.Number, ID := stat_id_new.BestEscapeStreak, Value := 1.0, InsertionMethod := stat_insertion_method.Highest}
                    stat_data{Player := Player, Category := stat_category.General, Type := stat_type_new.Number, ID := stat_id_new.ShortestEscape, Value := GoalDuration*1.0, InsertionMethod := stat_insertion_method.Lowest}
                    stat_data{Player := Player, Category := stat_category.Run, Type := stat_type_new.Number, ID := stat_id_new.Escaped, Value := 1.0}
                    }
                    if(SendStats[Data]){}

                GoalTimerDevice.Complete(Player)
                GoalCompletedAccolade.Award(Player)

                

            loop:
                Player:player = RoomManager.OnReplayGoal.Await()
                Logger.Print("Player replay")

                ReplayAnalytics.Submit(Player)
                ReplayAccolade.Award(Player)
                

    GetCompleteGoalMessage(GoalDuration:int)<computes>:string = "{CompleteGoalString} {TimeInSecondsToString(GoalDuration)}\n{CompleteGoalString2}"

    OnReplayButtonInteracted(Agent:agent):void =
        if (Player:=player[Agent], RoomManager.HasPlayerCompletedGoal(Player)):
            Logger.Print("Player asked to replay")

            if (FortCharacter := Agent.GetFortCharacter[]):
                FortCharacter.Heal(FortCharacter.GetMaxHealth())
                FortCharacter.Heal(FortCharacter.GetMaxHealth())
                
            ReplayTeleporter.Teleport(Player)

    SendStats<override>(Data:[]stat_data)<decides><transacts>:void = 
        if (Manager := StatsManager?):
            Manager.ReceiveData[Data]

    GetStatManager<override>()<decides><transacts>:stats_manager = 
        TaggedObjects := FindCreativeObjectsWithTag(stat_manager_tag{})
            var Objects:[]creative_object_interface=array{}
                for (Obj:TaggedObjects, TaggedObject:=creative_device_base[Obj]):
                    set Objects += array{TaggedObject}
                    #return Objects 
        Managers := for (Obj : TaggedObjects, Manager := stats_manager[Obj]):
            Manager
        if (Result := Managers[0]):
            Result
        else:
            Print("No stats manager found! returning empty instance")
            stats_manager{}
    
    OnAgentRevived(Agent:agent):void =
        if (Player := player[Agent]):
            Players := GetPlayspace().GetPlayers()
            var Data:[]stat_data = array{}
            for (OtherPlayer : Players):
                if:
                    Player <> OtherPlayer
                    OtherCharacter := OtherPlayer.GetFortCharacter[]
                    Character := Player.GetFortCharacter[]
                then:
                    OtherLocation := OtherCharacter.GetTransform().Translation
                    PlayerLocation := Character.GetTransform().Translation
                    PlayersDistance := Distance(OtherLocation, PlayerLocation)
                    if (PlayersDistance <= RevivalDistance):
                        set Data += array {
                            stat_data{Player := OtherPlayer, Category := stat_category.Run, Type := stat_type_new.Number, ID := stat_id_new.TeammatesRevived, Value := 1.0}
                            stat_data{Player := OtherPlayer, Category := stat_category.General, Type := stat_type_new.Number, ID := stat_id_new.TeammatesRevived, Value := 1.0}
                            } 
                    Logger.Print("Distance to player: {PlayersDistance}")
                    
            set Data += array {
                stat_data{Player := Player, Category := stat_category.General, Type := stat_type_new.Number, ID := stat_id_new.TimesRevived, Value := 1.0}
                stat_data{Player := Player, Category := stat_category.Run, Type := stat_type_new.Number, ID := stat_id_new.TimesRevived, Value := 1.0}
                }
            if(SendStats[Data]):
        
