using { /Fortnite.com/Devices }
using { /Verse.org/Simulation }
using { /UnrealEngine.com/Temporary/Diagnostics }

log_room_manager := class(log_channel){}

<# 
Data for a player related to the rooms
#>
player_room_data := class:
    var RoomIndexes<private>:[]int = array{}
    # Last room the player entered too
    var LastRoomIndex:int = INDEX_NONE

    var bHasStartedGoal:logic = false
    var bHasCompletedGoal:logic = false

    # Returns the index of the room the player is at. INDEX_NONE if none
    GetCurrentRoomIndex()<transacts>:int =
        if (LastIndex := RoomIndexes[RoomIndexes.Length-1]):
            return LastIndex
        INDEX_NONE

    # Adds a new room index, returns true if success
    AddRoomIndex(Index:int)<transacts><decides>:void =
        if (not RoomIndexes.Find[Index]):
            UpdateLastRoomIndex()
            set RoomIndexes += array{Index}
        else:
            false?

    # Removes a room index, returns true if success
    RemoveRoomIndex(Index:int)<transacts><decides>:void =
        if (RoomIndexes.Find[Index]):
            UpdateLastRoomIndex()
            set RoomIndexes = RoomIndexes.RemoveAllElements(Index)
        else:
            false?
        
    # Only update last room index if it is not INDEX_NONE
    UpdateLastRoomIndex()<transacts>:void =
        if (Index := GetCurrentRoomIndex(), Index <> INDEX_NONE):
            set LastRoomIndex  =Index

<# 
Keeps references and all related to rooms/aisles
#>
room_manager := class(creative_device, stat_sender):
    Logger:log = log{Channel:=log_room_manager}

    @editable # Reference of all the rooms in the map. Keep in order if you want to use player goals
    Rooms:[]room_volume = array{}

    @editable # Messsage device to show the current room the player is in
    RoomHUDMessage:hud_message_device = hud_message_device{}

    @editable
    CurrentRoomString:string = "Room "

    @editable # Know if tracking player goal to notify start and complete
    bTrackPlayerGoal:logic = true

    @editable # Index of the first room player goal
    RoomIndexStartGoal:int = 0

    @editable # Index of the last room player goal
    RoomIndexCompleteGoal:int = 1

    OnGoalStarted:event(player) = event(player){}
    OnGoalCancelled:event(player) = event(player){}
    OnGoalCompleted:event(player) = event(player){}
    OnReplayGoal:event(player) = event(player){}

    OnPlayerEnteredRoom:event(tuple(player, int)) = event(tuple(player, int)){}
    OnPlayerEnteredRoomBeforeGoal:event(player) = event(player){}

    # Map with the player and its data
    var PlayerRoomData:[player]player_room_data = map{}

    @editable
    NPCManager:enemies_manager = enemies_manager{}

    @editable
    AllItemsRemover:item_remover_device = item_remover_device{}

    var StatsManager:?stats_manager = false

    OnBegin<override>()<suspends>:void =
        if(FoundStatsManager := GetStatManager[]):
            set StatsManager = option{FoundStatsManager}

        for (Index := 0..(Rooms.Length-1)):
            spawn{AwaitRoomEvents(Index)}

        sync:
            loop:
                # Validate when all NPCs killed to unlock next door
                StageIndex:int = NPCManager.OnStageEmptyEvent.Await()
                if:
                    CleanedRoomIndex := GetIndexRoomWithNPCStageIndex(StageIndex)
                    CleanedRoom := Rooms[CleanedRoomIndex]
                    NextRoomIndex := CleanedRoomIndex+1
                    NextRoom := Rooms[NextRoomIndex]
                then:
                    CleanedRoom.OnRoomNPCEmptyStateChanged(true)
                    NextRoom.UnlockDoor()

                    # After unlock give reward to the players inside the room
                    <#
					if (NextRoom.RoomRewards.Length > 0):
                        for (PlayerDataKey->DataValue:PlayerRoomData):
                            if (DataValue.GetCurrentRoomIndex() = CleanedRoomIndex):
                                TryShowPreviewRoomReward(PlayerDataKey, NextRoomIndex, NextRoom)#>

                    # Prepare next room if using lock and empty
                    if (NextRoom.bUseLock? and not HasAnyPlayerCurrentRoomIndex(NextRoomIndex)?):
                        NPCManager.ActivateStage(NextRoom.NPCStageIndex)

            loop:
                # Validate when any NPC spawn to lock next door
                StageIndex:int = NPCManager.OnStageNotEmptyEvent.Await()
                if:
                    NotEmptyRoomIndex := GetIndexRoomWithNPCStageIndex(StageIndex)
                    NotEmptyRoom := Rooms[NotEmptyRoomIndex]
                    NextRoom := Rooms[NotEmptyRoomIndex+1]
                then:
                    NotEmptyRoom.OnRoomNPCEmptyStateChanged(false)
                    NextRoom.LockDoor()

    RegisterPlayer(Player:player):void =
        if (not PlayerRoomData[Player], set PlayerRoomData[Player] = player_room_data{}):
            # Validate if the player started inside a room
            for (Index := 0..(Rooms.Length-1), Room := Rooms[Index]):
                if (Room.IsAgentInRoom(Player)?):
                    OnAgentEnteredRoom(Player, Index)

    UnregisterPlayer(Player:player):void =
        if (PlayerRoomData[Player]):
            CleaningLastRoomInPlayer(Player)

            var NewMap:[player]player_room_data = map{}
            for (KeyPlayer -> Data : PlayerRoomData, KeyPlayer <> Player):
                set NewMap = ConcatenateMaps(NewMap, map{KeyPlayer => Data})
            set PlayerRoomData = NewMap

        # Clean player from all rooms rewards
        ResetPlayerRoomRewardsData(Player)

    GetPlayerRoomData(Player:player)<decides><transacts>: player_room_data =
        PlayerRoomData[Player]

    AwaitRoomEvents(IndexRoom:int)<suspends>:void =
        if (Room := Rooms[IndexRoom]):
            sync:
                loop:
                    Agent:agent = Room.OnAgentEnteredRoomEvent.Await()
                    OnAgentEnteredRoom(Agent, IndexRoom)

                loop:
                    Agent:agent = Room.OnAgentExitedRoomEvent.Await()
                    OnAgentExitedRoom(Agent, IndexRoom)
                    
    OnAgentEnteredRoom(Agent:agent, RoomIndex:int):void =
        if (Player := player[Agent], Data := PlayerRoomData[Player], PreviousRoomIndex := Data.GetCurrentRoomIndex(), Data.AddRoomIndex[RoomIndex]):
            Logger.Print("On Agent Entered Room Index {RoomIndex}")
            OnRoomIndexesChanged(Player, PreviousRoomIndex, Data.GetCurrentRoomIndex())

            OnPlayerEnteredRoom.Signal((Player, RoomIndex))

            # If room is any room before goal then we need to remove all rewards and clean player from room rewards
            if (RoomIndex < RoomIndexStartGoal):
                AllItemsRemover.Remove(Player)
                ResetPlayerReceivedRewards(Player) # Do not reset the rewards, it will only be done when dead
                OnPlayerEnteredRoomBeforeGoal.Signal(Player)

            if (bTrackPlayerGoal?):
                if (Data.bHasStartedGoal?):
                    if(RoomIndex < RoomIndexStartGoal):
                        if (set PlayerRoomData[Player].bHasStartedGoal = false):
                            OnGoalCancelled.Signal(Player)

                            if (Data.bHasCompletedGoal?):
                                OnReplayGoal.Signal(Player)
                                RestartGoalInPlayer(Player)

                    if (not Data.bHasCompletedGoal? and RoomIndex >= RoomIndexCompleteGoal):
                        if (set PlayerRoomData[Player].bHasCompletedGoal = true):
                            OnGoalCompleted.Signal(Player)
                else:
                    if (not Data.bHasCompletedGoal? and RoomIndex >= RoomIndexStartGoal and RoomIndex < RoomIndexCompleteGoal):
                        if (set PlayerRoomData[Player].bHasStartedGoal = true):
                            OnGoalStarted.Signal(Player)
                #TODO FIX Last room bug again
            StatData := array {
                stat_data{Player := Player, Category := stat_category.Run, Type := stat_type_new.Number, ID := stat_id_new.LastRoom, Value := RoomIndex*1.0, CalculationMethod := stat_calculation_method.Newest, InsertionMethod := stat_insertion_method.Newest}
            }
            if(SendStats[StatData]):


    OnAgentExitedRoom(Agent:agent, RoomIndex:int):void =
        if (RoomIndex > INDEX_NONE, Player := player[Agent], Data := PlayerRoomData[Player], PreviousRoomIndex := Data.GetCurrentRoomIndex(), Data.RemoveRoomIndex[RoomIndex]):
            Logger.Print("On Agent Exited Room Index {RoomIndex}")
            OnRoomIndexesChanged(Player, PreviousRoomIndex, Data.GetCurrentRoomIndex())

    OnRoomIndexesChanged(Player:player, PreviousRoomIndex:int, NewRoomIndex:int):void =
        # Verify if current room changed, otherwise skip
        if (PreviousRoomIndex = NewRoomIndex):
            return

        UpdateRoomHUDForPlayer(Player)

        if (NewRoomIndex > INDEX_NONE):
            # Notify accolade
            if (NewRoom := Rooms[NewRoomIndex]):
                NewRoom.PlayerEnteredRoomAnalytics.Submit(Player)

            # Reset next room if player moving around rooms, or not locker
            if (NextRoom := Rooms[NewRoomIndex+1], ShouldResetRoom(Player, NextRoom, NewRoomIndex+1)?):
                NPCManager.ActivateStage(NextRoom.NPCStageIndex)

            # Try clean when enters or exists to ensure empty states are updated
            CleaningLastRoomInPlayer(Player)

    ShouldResetRoom(Player:player, Room:room_volume, Index:int)<transacts>:logic =
        if (HasAnyPlayerCurrentRoomIndex(Index)?):
            return false

        # Player is in INDEX_NONE and LastRoom is the previous one
        for (PlayerData:PlayerRoomData):
            if (PlayerData.GetCurrentRoomIndex() = INDEX_NONE):
                if (PlayerData.LastRoomIndex = Index or PlayerData.LastRoomIndex = (Index-1)):
                    return false

        if (Room.bUseLock?):
            if (Room.bIsLocked?):
                return false

            # Validate no players in previous room if I was previously in this room
            if (PlayerRoomData[Player].LastRoomIndex = Index):
                return false
            # If has any player besides me then do not reset
            if (HasAnyPlayerCurrentRoomIndexBesidesPlayer(Index-1, Player)?):
                return false
        # If not initial room
        else if ((Index-1) >= RoomIndexStartGoal):
            if (PlayerRoomData[Player].LastRoomIndex = Index and HasAnyPlayerCurrentRoomIndex(Index-1)?):
                return false
            if (HasAnyPlayerCurrentRoomIndexBesidesPlayer(Index-1, Player)?):
                return false

        return true

    # Cleans NPCs in player last room index
    CleaningLastRoomInPlayer(Player:player):void =
        # Verify player last room and deactivate this and the next one (releases NPCs to have more space in the available pool (Max 30 in island))
        if (Data := PlayerRoomData[Player], Data.LastRoomIndex <> INDEX_NONE, LastPlayerRoom := Rooms[Data.LastRoomIndex]):
            Logger.Print("Try cleaning last room {Data.LastRoomIndex} and next last room {Data.LastRoomIndex+1}")

            if (ShouldDeactivateRoom(LastPlayerRoom, Data.LastRoomIndex)?):
                Logger.Print("Cleaning last room {Data.LastRoomIndex}")
                NPCManager.DeactivateStage(LastPlayerRoom.NPCStageIndex)
            
            if (NextLastRoom := Rooms[Data.LastRoomIndex+1], ShouldDeactivateRoom(NextLastRoom, Data.LastRoomIndex+1)?):
                Logger.Print("Cleaning next last room {Data.LastRoomIndex+1}")
                NPCManager.DeactivateStage(NextLastRoom.NPCStageIndex)

    ShouldDeactivateRoom(Room:room_volume, Index:int)<transacts>:logic =
        if (HasAnyPlayerCurrentRoomIndex(Index)?):
            return false
        
        # If locked then just clean
        if (Room.bIsLocked?):
            return true

        # Previous room is not empty
        if (HasAnyPlayerCurrentRoomIndex(Index-1)?):
            return false

        # Player is in INDEX_NONE and LastRoom is the previous one
        for (PlayerData:PlayerRoomData):
            if (PlayerData.GetCurrentRoomIndex() = INDEX_NONE):
                if (PlayerData.LastRoomIndex = Index or PlayerData.LastRoomIndex = (Index-1)):
                    return false
        
        return true

    HasAnyPlayerCurrentRoomIndex(Index:int)<transacts>:logic =
        if (Index <= INDEX_NONE):
            return false
            
        for (PlayerData:PlayerRoomData):
            if (PlayerData.GetCurrentRoomIndex() = Index):
                return true
        return false

    HasAnyPlayerCurrentRoomIndexBesidesPlayer(Index:int, IgnorePlayer:player)<transacts>:logic =
        if (Index <= INDEX_NONE):
            return false
            
        for (PlayerDataKey->DataValue:PlayerRoomData, PlayerDataKey <> IgnorePlayer):
            if (DataValue.GetCurrentRoomIndex() = Index):
                return true
        return false

    GetPlayerCountInRoomIndex(Index:int)<transacts>:int =
        var Count:int = 0;

        if (Index <= INDEX_NONE):
            return Count
            
        for (PlayerData:PlayerRoomData):
            if (PlayerData.GetCurrentRoomIndex() = Index):
                set Count += 1
        return Count

    UpdateRoomHUDForPlayer(Player:player):void =
        var NewRoomString:string = ""
        
        if:
            CurrentRoomIndex := PlayerRoomData[Player].GetCurrentRoomIndex()
            CurrentRoomIndex > INDEX_NONE and CurrentRoomIndex >= RoomIndexStartGoal
            CurrentRoom := Rooms[CurrentRoomIndex]
        then:
            set NewRoomString = CurrentRoomString + CurrentRoom.DisplayName

        RoomHUDMessage.Show(Player, StringToMessage(NewRoomString))

    HasPlayerCompletedGoal(Player:player)<transacts>:logic =
        if (Data := PlayerRoomData[Player]):
            Data.bHasCompletedGoal
        false

    RestartGoalInPlayer(Player:player):void =
        if (set PlayerRoomData[Player].bHasCompletedGoal = false) {}

    GetIndexRoomWithNPCStageIndex(StageIndex:int)<transacts>:int =
        for (Index := 0..(Rooms.Length-1), Room := Rooms[Index]):
            if (Room.NPCStageIndex = StageIndex):
                return Index
        INDEX_NONE

    OnPlayerEliminated(Player:player):void =
        # We use last index to know the room, because the player has already left that room when event is notified
        if (DeadRoomIndex := PlayerRoomData[Player].LastRoomIndex, Room := Rooms[DeadRoomIndex]):
            Logger.Print("Player died in room {DeadRoomIndex}")
            Room.PlayerDiedInRoomAnalytics.Submit(Player)
            
        # Reset rewards calculated for the player
        ResetRewardsForPlayer(Player)

    TryShowPreviewRoomReward(Player:player, RoomIndex:int, Room:room_volume): void = {}


    ResetPlayerRoomRewardsData(Player:player): void =
        for (Room:Rooms):
            Room.ResetPlayerReceivedReward(Player)
            Room.ResetRewardsForPlayer(Player)

    ResetPlayerReceivedRewards(Player:player): void =
        for (Room:Rooms):
            Room.ResetPlayerReceivedReward(Player)

    ResetRewardsForPlayer(Player:player): void =
        for (Room:Rooms):
            Room.ResetRewardsForPlayer(Player)

    GetGuardsInRoom(RoomIndex:int)<transacts>: []agent =
        if (Room:=Rooms[RoomIndex]):
            NPCManager.GetGuardsInStage(Room.NPCStageIndex)
        else:
            array{}
    
    GetRoomVolumeDevice(RoomIndex:int)<transacts>:room_volume=
        var TargetRoom:room_volume= room_volume{}
        if(ValidRoom:= Rooms[RoomIndex]):
           set TargetRoom = ValidRoom
        return TargetRoom

    SendStats<override>(Data:[]stat_data)<decides><transacts>:void = 
        if (Manager := StatsManager?):
            Manager.ReceiveData[Data]

    GetStatManager<override>()<decides><transacts>:stats_manager = 
        TaggedObjects := FindCreativeObjectsWithTag(stat_manager_tag{})
            var Objects:[]creative_object_interface=array{}
                for (Obj:TaggedObjects, TaggedObject:=creative_device_base[Obj]):
                    set Objects += array{TaggedObject}
                    #return Objects 
        Managers := for (Obj : TaggedObjects, Manager := stats_manager[Obj]):
            Manager
        if (Result := Managers[0]):
            Result
        else:
            Print("No stats manager found! returning empty instance")
            stats_manager{}