using { /Fortnite.com/Devices }
using { /Fortnite.com/Characters }
using { /Verse.org/Simulation }
using { /UnrealEngine.com/Temporary/Diagnostics }
using { /Verse.org/Simulation/Tags }

ClashBlueRoomReward_S1 := class(tag){}
ClashRedRoomReward_S1 := class(tag){}
ClashBlueRoomReward_S2 := class(tag){}
ClashRedRoomReward_S2 := class(tag){}
ClashBlueRoomReward_S3 := class(tag){}
ClashRedRoomReward_S3 := class(tag){}
ClashBlueRoomReward_S4 := class(tag){}
ClashRedRoomReward_S4 := class(tag){}
ClashArenaRoomReward_S5 := class(tag){}

GetClashRoomRewardTag(RoomIndex:int)<decides><transacts>:tag = 
    case(RoomIndex):
        1  => ClashBlueRoomReward_S1{}
        9  => ClashRedRoomReward_S1{}
        2  => ClashBlueRoomReward_S2{}
        8  => ClashRedRoomReward_S2{}
        3  => ClashBlueRoomReward_S3{}
        7  => ClashRedRoomReward_S3{}
        4  => ClashBlueRoomReward_S4{}
        6  => ClashRedRoomReward_S4{}
        5  => ClashArenaRoomReward_S5{}

room_data_reward := class<concrete>:
    @editable
    RewardType:reward_type = reward_type.Weapon  

    @editable
    TypeSearch:type_search = type_search.All

room_data_team := class:
    RoomDataRewards:[]room_data_reward
    Team:team_id

rewardableRoom := interface:
    GetUIRewardableData(Player:player): []ui_rewardable_data 
    GetRoomIndex():int
    GetIsRoomZero():logic
    RemovePlayerReward(Agent:agent): void

type_search := enum {All, Team}

#=========================================================================================================================

room_reward_device := class(creative_device,rewardableRoom):

    @editable
    RoomDataReward:[]room_data_reward = array{}

    @editable
    RewardManager:reward_manager = reward_manager{}   
    
    @editable
    IsRoomZero<public>:logic = false

    var RoomIndex:int = 0

    var RewardsPerPlayer:[player][]ui_rewardable_data = map{}

#=========================================================================================================================

    Initialize(Index:int):void =
        if(IsRoomZero=false):
            set RoomIndex = Index
    
    #Use this method to grant a reward to a player
    GrantReward(Agent:agent):void =
        CurrentTeam := GetAgentTeam(Agent)
        UIRewardableData := RewardManager.ChooseRandomRewards(room_data_team{RoomDataRewards := RoomDataReward, Team := CurrentTeam})

        if (Player:=player[Agent]):
            #If the player already has a pending reward in this room and is not the room zero, it shows the pending reward.
            if:
                HavePendingReward(Agent)?
            then:
                Print("Player Have pending Reward")
                RewardManager.ShowUIReward(Player,Self,RoomIndex)
            else:
                Print("Player Does not have pending Reward")
                #If the player does not have a pending reward in this room, generates a new one.
                if (set RewardsPerPlayer[Player] = UIRewardableData):
                    RewardManager.ShowUIReward(Player,Self,RoomIndex)

    #Check if player has a pending reward in a specific room
    HavePendingRewardInRoom(Agent:agent, Index: int)<transacts>: logic=
        #Make sure that the room_reward_device has a tag
        if (TagToSearch := GetClashRoomRewardTag[Index]):
    
            RoomRewardArray := FindCreativeObjectsWithTag(TagToSearch)
    
            for (RoomRewards : RoomRewardArray):
                if:
                    RoomReward := room_reward_device[RoomRewards]
                    RoomReward.HavePendingReward(Agent) = true
                then:
                    return true
        
        return false

    #Removes pending player rewards in this room
    RemovePlayerReward<override>(Agent:agent): void=
        if (Player:=player[Agent]):
            if (AlreadyHaveRewards := RewardsPerPlayer[Player]):
                set RewardsPerPlayer = RewardsPerPlayer - Player

                #Hide preview Reward
                RewardManager.HideAnyPreviewRoomReward(Player)
                
    #Check if a player has a pending reward in this room                    
    HavePendingReward(Agent:agent)<transacts>: logic=
        if (Player:=player[Agent]):
            if (AlreadyHaveRewards := RewardsPerPlayer[Player]):
                return true

        return false
    
    GetIsRoomZero<override>()<transacts>: logic=
        return IsRoomZero


    GenerateRandomPendingReward(Player:player):void=
        CurrentTeam := GetAgentTeam(Player)
        UIRewardableData := RewardManager.ChooseRandomRewards(room_data_team{RoomDataRewards := RoomDataReward, Team := CurrentTeam})
        #If the player does not have a pending reward in this room, generates a new one.
        if (set RewardsPerPlayer[Player] = UIRewardableData){}

#=========================================================================================================================

    GetUIRewardableData<override>(Player:player): []ui_rewardable_data=
        var UIRewardableDatas:[]ui_rewardable_data = array{}
        if:
            CurrentRewards := RewardsPerPlayer[Player]
        then:
            set UIRewardableDatas = CurrentRewards
            return UIRewardableDatas
        return UIRewardableDatas

    GetRoomIndex<override>():int =
        return RoomIndex
        