
using { /Fortnite.com/Devices }
using { /Fortnite.com/Characters }
using { /Fortnite.com/UI }
using { /Verse.org/Simulation }
using { /Verse.org/Colors }
using { /UnrealEngine.com/Temporary/Diagnostics }
using { /UnrealEngine.com/Temporary/SpatialMath }
using { /UnrealEngine.com/Temporary/UI }
#Prev version of this device: SS: 455
room_tracker_manager := class(creative_device):

    @editable
    ScoreManager : score_manager_device = score_manager_device{}

    @editable
    RoomManager : room_manager = room_manager{}

    var AllPlayers:[]player=array{}

    OnBegin<override>()<suspends>:void =
        set AllPlayers = GetPlayspace().GetPlayers()
        #UnCommend for TESTONLY
        # if(Player0:=AllPlayers[0]):
        #     TempPlayers:[]player=array{Player0,Player0,Player0,Player0,Player0}
        #     set AllPlayers = TempPlayers

        for (Player : AllPlayers):
            if (PlayerUI := GetPlayerUI[Player]):
                NewCanvas:= ConstructCanvas()
                TextBlocks:=ConstructWidget(AllPlayers, NewCanvas)
                PlayerUI.AddWidget(NewCanvas)
                spawn{ AwaitRoomEvents(TextBlocks) }

        #Try to spawn the widget just for client 0 of every client, but doesn't work, just spawns widget for Host
        # if(Player0:=AllPlayers[0]):
        #     if (PlayerUI := GetPlayerUI[Player0]):
        #         NewCanvas:= ConstructCanvas()
        #         TextBlocks:=ConstructWidget(AllPlayers, NewCanvas)
        #         PlayerUI.AddWidget(NewCanvas)
        #         spawn{ AwaitRoomEvents(TextBlocks) }
        
    AwaitRoomEvents(TextBlocks : []text_block)<suspends>:void=
        sync:
            loop:
                Player := RoomManager.OnPlayerEnteredRoomBeforeGoal.Await()
                if (PlayerIndex := AllPlayers.Find[Player]):
                    UpdatePlayerRoomInfo(Player, PlayerIndex, TextBlocks, "Main")

            loop:
                EnteredRoomResult:=RoomManager.OnPlayerEnteredRoom.Await()

                Player:= EnteredRoomResult(0)
                RoomIndex:int=EnteredRoomResult(1)
                RoomVolumeDevice:=RoomManager.GetRoomVolumeDevice(RoomIndex)

                if (PlayerIndex := AllPlayers.Find[Player]):
                    UpdatePlayerRoomInfo(Player, PlayerIndex, TextBlocks, RoomVolumeDevice.DisplayName)

    #Check widget's references in: WBP_Room_Tracker
    ConstructCanvas<protected>(): canvas =
        Canvas:=canvas:
            Slots:=array:
                canvas_slot:
                    Widget:= overlay:
                        Slots:=array:
                            overlay_slot:
                                VerticalAlignment:=vertical_alignment.Fill
                                HorizontalAlignment:=horizontal_alignment.Fill
                                Widget:= texture_block{DefaultImage:=Textures.General.T_BlueBg, DefaultTint:= NamedColors.Cyan}
 
                    Anchors := anchors{ Minimum := vector2{ X := 1.0, Y := 0.0}, Maximum := vector2{ X := 1.0, Y := 0.0}}
                    Offsets := margin {Left:= 0.0 , Top:= 315.0, Right:= 330.0, Bottom:= 200.0}
                    Alignment := vector2{X:= 1.0, Y:=0.0}
                    SizeToContent := false
                    

    ConstructWidget(Players : []player, Canvas:canvas):[]text_block=
        var TextBlocks:[]text_block=array{}

        if (Slot:=Canvas.Slots[0], Overlay:=overlay[Slot.Widget]):   

            StackBoxWidget:=stack_box:
                Orientation := orientation.Vertical 

            MyOverlay2:= overlay_slot:
                VerticalAlignment:=vertical_alignment.Fill
                HorizontalAlignment:=horizontal_alignment.Fill
                Widget:=StackBoxWidget  
            Overlay.AddWidget(MyOverlay2)
            
            for (Player : Players):
                TextWidget : text_block = text_block{DefaultTextColor := NamedColors.White}
                TextWidget.SetText(PlayerNameToMessage(Player, "Main"))

                set TextBlocks+=array{TextWidget}

                MyOverlay:= stack_box_slot:
                    VerticalAlignment:=vertical_alignment.Fill
                    HorizontalAlignment:=horizontal_alignment.Fill
                    Padding:= margin {Left:= 10.0 , Top:= 0.0, Right:= 10.0, Bottom:= 0.0}
                    Widget := TextWidget

                StackBoxWidget.AddWidget(MyOverlay)

        return TextBlocks

    PlayerNameToMessage<localizes>(Agent:agent, RoomName : string):message="{Agent} - {RoomName}"

    UpdatePlayerRoomInfo(Agent: agent, PlayerIndex: int, TextBlocks : []text_block, RoomName : string):void=
        if(Target:= TextBlocks[PlayerIndex]):
            Target.SetText(PlayerNameToMessage(Agent, RoomName))
