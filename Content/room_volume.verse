
using { /Fortnite.com/Devices }
using { /Verse.org/Simulation }
using { /Verse.org/Simulation/Tags }
using { /UnrealEngine.com/Temporary/Diagnostics }

log_room_volume := class(log_channel){}

<# 
Volume used for any room/aisle in the map
Allows to keep track of what room is the current player at
Has data for a specific room
#>
room_volume := class(creative_device):
    Logger:log = log{Channel:=log_room_volume}

    @editable
    DisplayName<public>:string = "0"

    @editable # Trigger used by this room volume
    Volume:volume_device = volume_device{}

    @editable
    bUseLock:logic = true

    @editable
    Lock:lock_device = lock_device{}

    var bIsLocked:logic = false

    @editable # Stage used for enemies in this room. INDEX_NONE if no NPCs. Use a different stage for each room
    NPCStageIndex:int = INDEX_NONE

    # Record of the players to whom rewards have been given
    var ReceivedRewardPlayers:[]player = array{}

    @editable
    CheckPointDevice:player_checkpoint_device=player_checkpoint_device{}

    @editable
    PlayerEnteredRoomAnalytics:analytics_device = analytics_device{}

    @editable
    PlayerDiedInRoomAnalytics:analytics_device = analytics_device{}

    var AgentsInVolume:[]agent = array{}

    OnAgentEnteredRoomEvent<public>:event(agent) = event(agent){}

    OnAgentExitedRoomEvent<public>:event(agent) = event(agent){}

    OnBegin<override>()<suspends>:void=
        Volume.AgentEntersEvent.Subscribe(OnAgentEntersRoom)
        Volume.AgentExitsEvent.Subscribe(OnAgentExitsRoom)

    OnAgentEntersRoom(Agent:agent):void =
        if (not AgentsInVolume.Find[Agent]):
            set AgentsInVolume += array{Agent}

            OnAgentEnteredRoomEvent.Signal(Agent)

    OnAgentExitsRoom(Agent:agent):void =
        LastLength:int = AgentsInVolume.Length

        set AgentsInVolume = AgentsInVolume.RemoveAllElements(Agent)

        if (LastLength <> AgentsInVolume.Length):
            OnAgentExitedRoomEvent.Signal(Agent)

    IsAgentInRoom(Agent:agent)<transacts>:logic =
        if (AgentsInVolume.Find[Agent]) then true else false

    IsEmpty()<transacts>:logic =
        if (AgentsInVolume.Length <= 0) then true else false

    LockDoor():void =
        if (bUseLock?, FirstPlayer := GetPlayspace().GetPlayers()[0]):
            Logger.Print("Door locked {DisplayName}")
            Lock.Close(FirstPlayer)
            Lock.Lock(FirstPlayer)
            set bIsLocked = true
            
            # When room locked, reset the rewards
            ResetAllRewardsPerPlayer()

    UnlockDoor():void =
        if (bUseLock?, FirstPlayer := GetPlayspace().GetPlayers()[0]):
            Logger.Print("Door unlocked {DisplayName}")
            Lock.Unlock(FirstPlayer)
            Lock.Open(FirstPlayer)
            set bIsLocked = false

    # Only give reward to players once
    CanGiveRewardToPlayer(Player:player)<decides><transacts>:void =
        not ReceivedRewardPlayers.Find[Player]

    AddReceivedRewardPlayer(Player:player):void =
        set ReceivedRewardPlayers += array{Player}

    ResetPlayerReceivedReward(Player:player):void =
        NewReceivedRewards := ReceivedRewardPlayers.RemoveAllElements(Player)

        if (ReceivedRewardPlayers.Length <> NewReceivedRewards.Length):
            set ReceivedRewardPlayers = NewReceivedRewards

    ResetRewardsForPlayer(Player:player):void = {}

    ResetAllRewardsPerPlayer():void = {}

    ResetActiveBoosters(): void = {}

    OnRoomNPCEmptyStateChanged(bIsEmpty:logic): void = {}