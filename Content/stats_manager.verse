using { /Fortnite.com/Devices }
using { /Verse.org/Simulation }
using { /UnrealEngine.com/Temporary/Diagnostics }
using { /Verse.org/Simulation/Tags }
using { /Verse.org/Colors/NamedColors }
using { /Verse.org/Assets }

#This is temporary, meant to be copy/pasted into the classes that will be sending stats and inheriting from the stat_sender interface, see the enemies manager
# SendStats<override>(Data:[]stat_data)<decides><transacts>:void = 
#     Manager := GetStatManager[]
#     Manager.ReceiveData[Data]

# GetStatManager<override>()<decides><transacts>:stats_manager = 
#     Objs := GetCreativeObjectsWithTag(stat_manager_tag{})
#     Managers := for (Obj : Objs, Manager := stats_manager[Obj]):
#         Manager
#     if (Result := Managers[0]):
#         Result
#     else:
#         Print("No stats manager found! returning empty instance")
#         stats_manager{}

# Data := array {
        # stat_data{Player := , Category := stat_category., Type := stat_type_new., ID := stat_id_new., Value := }
        # }
# if(SendStats[Data]):

#Type alias, not needed anymore but can´t be deleted because it's in an uploaded version already
stat_ratio := tuple(float, float)

log_stats_manager := class(log_channel){}

# Variable to define if a stat is a general or a run one
stat_category := enum{None, Run, General}

# DEPRECATED, DO NOT USE, USE stat_type_new INSTEAD
stat_type := enum<persistable>{None, Number, Logic, Text, Ratio}

# This is to help understand how to resolve data to show in the UI since it's always just a float in the persistable data
stat_type_new := enum{None, Number, Logic, Text, ItemID, BoosterID }

# This array is used to map each value to a number, this HAS to be updated if the enum is changed
StatTypeArray:[]stat_type_new = array{
    stat_type_new.Number
    stat_type_new.Logic
    stat_type_new.Text
    stat_type_new.ItemID
    stat_type_new.BoosterID
}

# This helps determine if the new value should be just added to the old one, or choose between of the two in the temporary stats before saving
stat_calculation_method := enum{Add, Highest, Lowest, Newest}

# This helps determine how to choose between the old stat and the new value when saving the persistable stat
stat_insertion_method := enum{Newest, Highest, Lowest}

stat_manager_tag := class(tag){}

ToString(Category:stat_category)<computes>:string =
    case(Category):
        stat_category.Run => "Run"
        stat_category.General => "General"
        stat_category.None => "None"

ToString(Type:stat_type_new)<computes>:string =
    case(Type):
        stat_type_new.Number => "Number"
        stat_type_new.Logic => "Logic"
        stat_type_new.Text => "Text"
        stat_type_new.ItemID => "Item ID"
        stat_type_new.BoosterID => "Booster ID"
        stat_type_new.None => "None"

# Enum to identify all stats, has an array to use indices
stat_id_new := enum{
    None
    EnemiesEliminated
    BossesEliminated
    Type1EnemiesEliminated
    Type2EnemiesEliminated
    Type3EnemiesEliminated
    Type4EnemiesEliminated
    Type5EnemiesEliminated
    Type1BossesEliminated
    Type2BossesEliminated
    Type3BossesEliminated
    Type4BossesEliminated
    Type5BossesEliminated
    HighestEliminationStreak
    MostSimultaneousEliminations
    ShortestEscape
    Escapes
    EscapeAttempts
    EscapeEffectiveness
    GuestsRescued
    EscapeStreak
    BestEscapeStreak
    CoinsEarned
    EliminationCoins
    GuestsCoins
    AssistCoins
    RestZoneCoins
    PrestigeLevel
    TeammatesRevived
    TimesRevived
    MostUsedBooster
    DiamondWizHighestMultiplier
    DiamondsEarned
    DiamondsSpent
    ContractsCompleted
    LongestRestZoneSession
    TotalPlaytime
    Playtime
    Rerolls
    Escaped
    LastRoom
    RoomRewards
}

# Array to hold all stat IDs, must be kept updated if the stat_id_new enum is modified
StatIDArray:[]stat_id_new = array{
    stat_id_new.None
    stat_id_new.EnemiesEliminated
    stat_id_new.BossesEliminated
    stat_id_new.Type1EnemiesEliminated
    stat_id_new.Type2EnemiesEliminated
    stat_id_new.Type3EnemiesEliminated
    stat_id_new.Type4EnemiesEliminated
    stat_id_new.Type5EnemiesEliminated
    stat_id_new.Type1BossesEliminated
    stat_id_new.Type2BossesEliminated
    stat_id_new.Type3BossesEliminated
    stat_id_new.Type4BossesEliminated
    stat_id_new.Type5BossesEliminated
    stat_id_new.HighestEliminationStreak
    stat_id_new.MostSimultaneousEliminations
    stat_id_new.ShortestEscape
    stat_id_new.Escapes
    stat_id_new.EscapeAttempts
    stat_id_new.EscapeEffectiveness
    stat_id_new.GuestsRescued
    stat_id_new.EscapeStreak
    stat_id_new.BestEscapeStreak
    stat_id_new.CoinsEarned
    stat_id_new.EliminationCoins
    stat_id_new.GuestsCoins
    stat_id_new.AssistCoins
    stat_id_new.RestZoneCoins
    stat_id_new.PrestigeLevel
    stat_id_new.TeammatesRevived
    stat_id_new.TimesRevived
    stat_id_new.MostUsedBooster
    stat_id_new.DiamondWizHighestMultiplier
    stat_id_new.DiamondsEarned
    stat_id_new.DiamondsSpent
    stat_id_new.ContractsCompleted
    stat_id_new.LongestRestZoneSession
    stat_id_new.TotalPlaytime
    stat_id_new.Playtime
    stat_id_new.Rerolls
    stat_id_new.Escaped
    stat_id_new.LastRoom
    stat_id_new.RoomRewards
}

#DEPRECATED, DO NOT USE, USE stat_id_new instead
stat_id := enum<persistable>{
    None
    EnemiesEliminated
    BossesEliminated
    Type1EnemiesEliminated
    Type2EnemiesEliminated
    Type3EnemiesEliminated
    Type4EnemiesEliminated
    Type5EnemiesEliminated
    Type1BossesEliminated
    Type2BossesEliminated
    Type3BossesEliminated
    Type4BossesEliminated
    Type5BossesEliminated
    HighestEliminationStreak
    MostLethalWeapon
    MostSimultaneousEliminations
    ShortestEscape
    Escapes
    EscapeAttempts
    EscapeEffectiveness
    GuestsRescued
    EscapeStreak
    BestEscapeStreak
    CoinsEarned
    CoinsFromGuests
    AssistCoins
    RestZoneCoins
    PrestigeLevel
    TeammatesRevived
    TimesRevived
    MostUsedBooster
    DiamondWizHighestMultiplier
    DiamondsEarned
    ContractsCompleted
    LongestRestZoneSession
    TotalPlaytime
    Rerolls
    Escaped
}

#Function to be able to print the ID value
ToString(ID:stat_id_new)<computes>:string =
    case(ID):
        stat_id_new.EnemiesEliminated =>            "Enemies Eliminated"
        stat_id_new.BossesEliminated =>             "Bosses Eliminated"
        stat_id_new.Type1EnemiesEliminated =>       "Type 1 Enemies Eliminated"
        stat_id_new.Type2EnemiesEliminated =>       "Type 2 Enemies Eliminated"
        stat_id_new.Type3EnemiesEliminated =>       "Type 3 Enemies Eliminated"
        stat_id_new.Type4EnemiesEliminated =>       "Type 4 Enemies Eliminated"
        stat_id_new.Type5EnemiesEliminated =>       "Type 5 Enemies Eliminated"
        stat_id_new.Type1BossesEliminated =>        "Type 1 Bosses Eliminated"
        stat_id_new.Type2BossesEliminated =>        "Type 2 Bosses Eliminated"
        stat_id_new.Type3BossesEliminated =>        "Type 3 Bosses Eliminated"
        stat_id_new.Type4BossesEliminated =>        "Type 4 Bosses Eliminated"
        stat_id_new.Type5BossesEliminated =>        "Type 5 Bosses Eliminated"
        stat_id_new.HighestEliminationStreak =>     "Highest Elimination Streak"
        stat_id_new.MostSimultaneousEliminations => "Most Simultaneous Eliminations"
        stat_id_new.ShortestEscape =>               "Shortest Escape"
        stat_id_new.Escapes =>                      "Escapes"
        stat_id_new.EscapeAttempts =>               "Escape Attempts"
        stat_id_new.EscapeEffectiveness =>          "Escape Effectiveness"
        stat_id_new.GuestsRescued =>                "Guests Rescued"
        stat_id_new.EscapeStreak =>                 "Current Escape Streak"
        stat_id_new.BestEscapeStreak =>             "Best Escape Streak"
        stat_id_new.CoinsEarned =>                  "Coins Earned"
        stat_id_new.EliminationCoins =>             "Coins from Eliminations"
        stat_id_new.GuestsCoins =>                  "Coins From Guests"
        stat_id_new.AssistCoins =>                  "Assist Coins"
        stat_id_new.RestZoneCoins =>                "Rest Zone Coins"
        stat_id_new.PrestigeLevel =>                "Prestige Level"
        stat_id_new.TeammatesRevived =>             "Teammates Revived"
        stat_id_new.TimesRevived =>                 "Times Revived"
        stat_id_new.MostUsedBooster =>              "Most Used Booster"
        stat_id_new.DiamondWizHighestMultiplier =>  "Diamond Wiz Highest Multiplier"
        stat_id_new.DiamondsEarned =>               "Diamonds Earned"
        stat_id_new.DiamondsSpent =>                "Diamonds Spent"
        stat_id_new.ContractsCompleted =>           "Contracts Completed"
        stat_id_new.LongestRestZoneSession =>       "Longest Rest Zone Session"
        stat_id_new.Playtime =>                     "Playtime"
        stat_id_new.Rerolls =>                      "Rerolls"
        stat_id_new.Escaped =>                      "Escaped"
        stat_id_new.LastRoom =>                     "Last Room"
        stat_id_new.RoomRewards =>                  "Room Rewards"
        stat_id_new.None =>                         "None"
        _ =>                                        "Undefined"
        

# Class sent from other classes for the stat manager to read all needed data and process it correctly
stat_data := class:

    Player:player

    Category: stat_category

    Type:stat_type_new = stat_type_new.Number

    ID:stat_id_new

    Value:float = 0.0

    CalculationMethod:stat_calculation_method = stat_calculation_method.Add

    InsertionMethod:stat_insertion_method = stat_insertion_method.Newest

ToString(Data:stat_data)<computes><reads>:string =
    "Category: {Data.Category}, ID: {Data.ID}, Type: {Data.Type}, Value: {Data.Value}"
    
# DEPRECATED, use stat_new instead
stat := class<concrete><final><persistable>:
    
    Type:stat_type = stat_type.None
    
    NewType:int = -1

    NumberValue:?float = false

    LogicValue:?logic = false

    TextValue:?string = false

    RatioValue:?stat_ratio = false


# Base class to store a stat, has a type index and a value
stat_new := class<concrete><final><persistable>:
    #This type is translated from stat_type_new enum values and array
    Type:int = -1

    Value:float = -1.0

# Class that stores all the stats for a player, both general and run stats as maps with indices that are translated from stat_id_new enum and array 
player_stats := class<final><persistable>:

    Version:int = 1
    #DEPRECATED, not used anymore, use new run stats
    RunStats:[stat_id]stat = map{}
    #DEPRECATED, not used anymore, use new general stats
    GeneralStats:[stat_id]stat = map{}

    NewRunStats:[int]stat_new = map{}

    NewGeneralStats:[int]stat_new = map{}

# Helper class that stores stats like the player stats but can be edited directly before using to save into the player stats
editable_player_stats := class:

    var NewRunStats:[int]stat_new = map{}

    var NewGeneralStats:[int]stat_new = map{}

# Constructor to make new player stats
MakePlayerStats<constructor>(Src:player_stats)<transacts> := player_stats:

    Version:= Src.Version

    RunStats:= map{}

    GeneralStats:= map{}

    NewRunStats:= Src.NewRunStats

    NewGeneralStats:= Src.NewGeneralStats

#Interface that must be added to any class that will send stats, defines the two methods used to send them
stat_sender := interface():
    SendStats(Data:[]stat_data)<decides><transacts>:void
    GetStatManager()<decides><transacts>:stats_manager

# Persistent map of stats for each player
var PlayerStats:weak_map(player,player_stats) = map{}

#Class that handles all stats, general and run stats, like eliminations, deaths, accumulated currency, etc
stats_manager := class(creative_device):

    Logger:log = log{Channel := log_stats_manager}
    # Var to set if we want to print debug messages
    @editable
    ShowDebug:logic = false

    #Var to device that shows the stats UI
    #@editable
    #StatsUI:ui_stats = ui_stats{}

    #Dialog device that sends the info to show the UI depending on the category (General or Run stats)
    @editable
    StatsDialog:popup_dialog_device = popup_dialog_device{}
 
    #Persistable data device, used to read some values to fill stats when they are created the first time 
    @editable
    PlayerPersistableData:player_persistable_data = player_persistable_data{}

    @editable
    TEST_BatchInsertTime:float = 2.0

    #Debug button to log stats to check without using the UI
    @editable
    DEBUG_Button:button_device = button_device{}

    #Debug variable to disable stat saving to prevent issues on update until feature is ready
    @editable
    DEBUG_EnableStatSaving:logic = false

    @editable
    DEBUG_GetRunCoinsButton:button_device = button_device{}

    #Helper variable to store and manipulate data before inserting into persistent stats
    var TempPlayerStats:[player]editable_player_stats = map{}


    # Runs when the device is started in a running game
    OnBegin<override>()<suspends>:void =
        GetPlayspace().PlayerAddedEvent().Subscribe(OnPlayerAdded)
        GetPlayspace().PlayerRemovedEvent().Subscribe(OnPlayerRemoved)
        Players := GetPlayspace().GetPlayers()
        for (Player : Players, CheckStatsDataForPlayer[Player]):

        DEBUG_Button.InteractedWithEvent.Subscribe(DEBUG_Print)
        DEBUG_GetRunCoinsButton.InteractedWithEvent.Subscribe(PrintRunCoins)

    PrintRunCoins(Agent:agent):void =
        if (Player := player[Agent]):
            Coins := GetFloatStatValue(Player, stat_category.Run, stat_id_new.CoinsEarned)
            Print("Run coins: {Coins}", ?Duration:=6.0)

    # Function to send a debug stat for testing purposes
    DEBUG_SendStat(Agent:agent):void =
        if(Player := player[Agent]):
            Data := array {
                stat_data{Player := Player, Category := stat_category.General, Type := stat_type_new.Number, ID := stat_id_new.None, Value := 10.0 }
                }
            if(ReceiveData[Data]):
                Print("Debug stat sent", ?Duration:=6.0)
            
    # Function to log stats         
    DEBUG_Print(Agent:agent):void =
        if (Player := player[Agent]):
            LogData(Player)
    
    #Initialize data for player in case they arrive mid match
    OnPlayerAdded(Player:player):void =
        if(CheckStatsDataForPlayer[Player]):
    
    #Clear Run stats when player leaves
    OnPlayerRemoved(Player:player):void = 
        if(ResetRunStats[Player, false]):
    
    #This function receives the data from all the other devices when sending stats, processes it, and then calls the function to insert the data persistently
    ReceiveData(NewDataArray:[]stat_data)<decides><transacts>:void =
        if (DEBUG_EnableStatSaving?):
            for (NewData : NewDataArray): 
                if(ShowDebug?) then Logger.Print("Received data: {NewData}")

                var NewStat:stat_new = stat_new{}
                if (ShowDebug?) then Logger.Print("Created empty stat")
                Player := NewData.Player
                if (ShowDebug?) then Logger.Print("Player Set")
                var EditableData:editable_player_stats = TempPlayerStats[Player]
                if (ShowDebug?) then Logger.Print("Editable Data Set")
                var NewValue:float = NewData.Value
                if (NewData.Category = stat_category.Run):
                    if (OldValue := EditableData.NewRunStats[GetIndexFromID(NewData.ID, StatIDArray)].Value):
                        MaybeNewValue := ProcessValue[OldValue, NewData.Value, NewData.CalculationMethod]
                        set NewValue = MaybeNewValue  
                    else:
                        Logger.Print("No old value")
                    set NewStat = stat_new{Type:= GetIndexFromID(NewData.Type, StatTypeArray), Value := NewValue}
                    set EditableData.NewRunStats[GetIndexFromID(NewData.ID, StatIDArray)] = NewStat
                    if (ShowDebug?) then Logger.Print("Editable run stat set")

                else:
                    if (OldValue := EditableData.NewGeneralStats[GetIndexFromID(NewData.ID, StatIDArray)].Value):
                        MaybeNewValue := ProcessValue[OldValue, NewData.Value, NewData.CalculationMethod]
                        set NewValue = MaybeNewValue
                    else:
                        Logger.Print("No old value")
                    set NewStat = stat_new{Type:= GetIndexFromID(NewData.Type, StatTypeArray), Value := NewValue}
                    set EditableData.NewGeneralStats[GetIndexFromID(NewData.ID, StatIDArray)] = NewStat
                    if (ShowDebug?) then Logger.Print("Editable general stat set")

                
                InsertPersistentStats[Player, EditableData, ?InsertionMethod := NewData.InsertionMethod]
        else:
            if (ShowDebug?) then Logger.Print("Stat saving disabled for debug")

    # Constructor to instance editable stats               
    MakeEditableStats(Stats:player_stats)<transacts>:editable_player_stats =
        editable_player_stats{NewRunStats := Stats.NewRunStats, NewGeneralStats := Stats.NewGeneralStats}
    
    # Function to initialize persistent and editable stats with empty variables if they don't exist yet
    CheckStatsDataForPlayer(Player:player)<decides><transacts>:void =
        if (PlayerStats[Player]):
            #SetStartingValues(Player)
        else:
            set PlayerStats[Player] = player_stats{}
        set TempPlayerStats[Player] = editable_player_stats{}

    
    #Function to calculate the new value depending on the calculation method provided
    ProcessValue(OldValue:float, NewValue:float, CalculationMethod:stat_calculation_method)<decides><transacts>:float =
        case(CalculationMethod):
            stat_calculation_method.Add => OldValue + NewValue
            stat_calculation_method.Highest => Max(OldValue, NewValue)
            stat_calculation_method.Lowest => Min(OldValue, NewValue)
            stat_calculation_method.Newest => NewValue
    
    #Function to process editable stats, compares to the saved data and chooses the appropiate value depending on the insertion method provided, or if there is no saved value only inserts the new value directly
    InsertPersistentStats(Player:player, EditableStats:editable_player_stats, ?InsertionMethod:stat_insertion_method = stat_insertion_method.Newest)<decides><transacts>:void =
        
        if (ShowDebug?) then Logger.Print("Entering insert persistent data")
        var StaticStats:player_stats = player_stats{}
        if (ShowDebug?) then Logger.Print("Found Static Stats")
        var LocalStats:editable_player_stats = EditableStats
        if (SavedStats := PlayerStats[Player]):
            set StaticStats = SavedStats
            if (ShowDebug?) then Logger.Print("Found Saved stats for player")
                 
        for (ID -> EditableRunStat: EditableStats.NewRunStats):
            if:
                SavedStat := StaticStats.NewRunStats[ID]
            then:
                if (ShowDebug?) then Logger.Print("Found Saved Stat")
                var LocalValue:float = EditableRunStat.Value
                
                SavedValue := SavedStat.Value
            
                case(InsertionMethod):
                    stat_insertion_method.Highest => 
                        set LocalValue = Max(SavedValue, LocalValue)
                    stat_insertion_method.Lowest =>
                        set LocalValue = Min(SavedValue, LocalValue)
                    stat_insertion_method.Newest => {}
                set LocalStats.NewRunStats[ID] = stat_new{Type:= EditableRunStat.Type, Value := LocalValue}

            else:
                if (ShowDebug?) then Logger.Print("Couldn't find saved data for stat {ID}")
          
                LocalValue:float = EditableRunStat.Value
                set LocalStats.NewRunStats[ID] = stat_new {Type:= EditableRunStat.Type, Value := LocalValue}

        for (ID -> EditableGeneralStat: EditableStats.NewGeneralStats):
            if:
                SavedStat := StaticStats.NewGeneralStats[ID]
                Type := GetIDFromIndex[EditableGeneralStat.Type, StatTypeArray]
            then:
                var LocalValue:float = EditableGeneralStat.Value
                SavedValue := SavedStat.Value
        
                case(InsertionMethod):
                    stat_insertion_method.Highest => 
                        set LocalValue = Max(SavedValue, LocalValue)
                    stat_insertion_method.Lowest =>
                        set LocalValue = Min(SavedValue, LocalValue)
                    stat_insertion_method.Newest => {}

                set LocalStats.NewGeneralStats[ID] = stat_new {Type:= EditableGeneralStat.Type, Value := LocalValue}
            
            else:
                # Print("Couldn't find saved data for stat {ID}", ?Duration:=6.0)
                LocalValue:float = EditableGeneralStat.Value
                set LocalStats.NewGeneralStats[ID] = stat_new {Type:= EditableGeneralStat.Type, Value := LocalValue}
        # Important! this block makes the updated data first, checks if the new data will fit and only then inserts it in the saved data, this prevents verse crashing if the data doesn't fit
        if:
            var UpdatedStats: player_stats = StaticStats
            set UpdatedStats = UpdateStats(StaticStats, LocalStats)
            FitsInPlayerMap[UpdatedStats]
            set PlayerStats[Player] = player_stats:
                MakePlayerStats<constructor>(StaticStats)
                NewRunStats := UpdatedStats.NewRunStats
                NewGeneralStats := UpdatedStats.NewGeneralStats
        then:
            if (ShowDebug?) then Print("Persistent stats saved", ?Duration:=6.0)
        else:
            #This is too important to not register even if we don't want to show debug messages
            Logger.Print("Stat data doesn't fit")
            Print("Stat data doesn't fit", ?Duration:=6.0, ?Color := Red)
                
    #Function that deletes all the run data and inserts it, this should happen when a player dies or leaves
    ResetRunStats(Player:player, ResetPersistentData:logic)<decides><transacts>:void =
        if:
            var TempStats:editable_player_stats = TempPlayerStats[Player]
        then:
            set TempStats.NewRunStats = map{}
            if (ResetPersistentData?):
                InsertPersistentStats[Player, TempPlayerStats[Player]]
            if (ShowDebug?) then Print("Run stats reset", ?Duration:=6.0)

    #Function to reset any stat to a default value
    ResetStat(Player:player, Category:stat_category, ID: stat_id_new)<decides><transacts>:void = 
        EmptyStat := stat_new{}
        if:
            TempStat := TempPlayerStats[Player]
            if (Category = stat_category.General):
                set TempStat.NewGeneralStats[GetIndexFromID(ID, StatIDArray)] = EmptyStat
            else:
                set TempStat.NewRunStats[GetIndexFromID(ID, StatIDArray)] = EmptyStat
            InsertPersistentStats[Player, TempStat]

    #Function to log all the stats to screen
    LogData(Player:player):void =
        if:
            Stats := PlayerStats[Player]
        then:
            for (ID -> Stat : Stats.NewGeneralStats, Type := GetIDFromIndex[Stat.Type, StatTypeArray], EnumID := GetIDFromIndex[ID, StatIDArray]):
                Logger.Print("Printing Data for general stat {EnumID}")
                
                Value := Stat.Value
                Logger.Print("Number Value: {Value}, Type is {Type}")

            Logger.Print("There are {Stats.NewGeneralStats.Length} general stats saved")

            for (ID -> Stat : Stats.NewRunStats, Type := GetIDFromIndex[Stat.Type, StatTypeArray], EnumID := GetIDFromIndex[ID, StatIDArray]):
                Logger.Print("Printing Data for run stat {EnumID}")
                
                Value := Stat.Value
                Logger.Print("Number Value: {Value}, Type is {Type}")

            Logger.Print("There are {Stats.NewRunStats.Length} run stats saved")
    
    #Function to retrieve the value of any stat
    GetFloatStatValue(Player:player, Category:stat_category, ID:stat_id_new)<transacts>:float = 
        if:
            FullStats := PlayerStats[Player]
            Stats := if (Category = stat_category.General) then FullStats.NewGeneralStats else FullStats.NewRunStats
            Stat := Stats[GetIndexFromID(ID, StatIDArray)]
            #Stat := Stats[GetStatIDIndex(ID)]
            #Stat.Type = stat_type_new.Number
            Value := Stat.Value
        then:
            Value
        else: 
            -1.0
    
    #Important! This function created new stat data combining old and new stats, then making a new variable with that data and returns it
    UpdateStats(Src:player_stats, NewStats:editable_player_stats)<transacts>:player_stats = 
        UpdatedRunStats:[int]stat_new = MergeMaps(Src.NewRunStats, NewStats.NewRunStats)
        UpdatedGeneralStats:[int]stat_new = MergeMaps(Src.NewGeneralStats, NewStats.NewGeneralStats)

        UpdatedStats:player_stats = player_stats:
            MakePlayerStats<constructor>(Src)
            NewRunStats := UpdatedRunStats
            NewGeneralStats := UpdatedGeneralStats

    <#SetStartingValues(Player:player)<transacts>:void = 
        var Data:[]stat_data = array{}
        if (Prestige := PlayerPersistableData.GetPlayerPrestigeLevel[Player], GetFloatStatValue(Player, stat_category.General, stat_id_new.PrestigeLevel) = -1):
            set Data += array{
                stat_data{Player := Player, Category := stat_category.General, Type := stat_type_new.Number, ID := stat_id_new.PrestigeLevel, Value := Prestige*1.0, CalculationMethod := stat_calculation_method.Newest, InsertionMethod := stat_insertion_method.Highest}
            }
        if (GetFloatStatValue(Player, stat_category.General, stat_id_new.CoinsEarned) = -1):
            CurrentGold := PlayerPersistableData.GetPlayerCurrency(Player, currency_id.Gold)
            set Data += array{
                    stat_data{Player := Player, Category := stat_category.General, Type := stat_type_new.Number, ID := stat_id_new.CoinsEarned, Value := CurrentGold}
                }
        if (GetFloatStatValue(Player, stat_category.General, stat_id_new.DiamondsEarned) = -1):
            CurrentDiamonds := PlayerPersistableData.GetPlayerCurrency(Player, currency_id.Diamond)
            set Data += array{
                    stat_data{Player := Player, Category := stat_category.General, Type := stat_type_new.Number, ID := stat_id_new.DiamondsEarned, Value := CurrentDiamonds}
                }

        if(ReceiveData[Data]):#>


    BatchInsertData()<suspends>:void = 
        for (Player -> EditableData : TempPlayerStats, InsertPersistentStats[Player, EditableData]):
            Sleep(TEST_BatchInsertTime)
            






   
