using { /Fortnite.com/Characters }
using { /Fortnite.com/Devices }
using { /Fortnite.com/Game }
using { /UnrealEngine.com/Temporary/Diagnostics }
using { /Verse.org/Simulation }
using { /Verse.org/Colors }
using { /Verse.org/Colors/NamedColors }

#Currencies are granted on enemies_manager class


streaks_manager := class(creative_device):

    @editable
    WaitForCinematic : logic = true

    @editable
    var PlayerTempData : player_temp_data = player_temp_data{}

    @editable
    PlayerPersistableData:player_persistable_data = player_persistable_data{}

    @editable
    var EliminationManagerAI : elimination_manager_device = elimination_manager_device{}

    @editable
    RoomManager:room_manager = room_manager{}

    @editable # If WaitForCinematic is true, the guest rescue manager will only be initialized after this cinematic stops
    CinematicSequenceTrigger : cinematic_sequence_device = cinematic_sequence_device{}

    @editable
    MaxKillStreak:int=500

    @editable
    MaxKillStreakMultiplier:float=50.0

    @editable
    KillStreakMultiplier : float = 0.1

    #Increase the Value of => MaxKillStreak
    @editable
    PrestigeIncreaseMaxKS : int = 100

    #Increase the Value of => MaxKillStreakMultiplier
    @editable
    PrestigeIncreaseMaxMulti : float = 10.0

    @editable
    PrestigeAdditionalMulti : float = 0.1

    @editable
    EnableKillClamp:logic=false

    var KillStreakMap : [player]int = map{}

    #var MaybeTrackerPerPlayer : [player]?ui_killstreak_counter = map{}

    # -------- INITIALIZATION --------
    OnBegin<override>()<suspends>:void=
        if (WaitForCinematic?):
            CinematicSequenceTrigger.StoppedEvent.Await()

        EliminationManagerAI.EliminationEvent.Subscribe(OnAIElimination)

        InitializeUI()

        GetPlayspace().PlayerAddedEvent().Subscribe(OnPlayerJoined)

        sync:
            loop:
                DamageResult := PlayerTempData.OnPlayerDamaged.Await()
                OnCharacterDamaged(DamageResult)

            loop:
                Player := RoomManager.OnPlayerEnteredRoomBeforeGoal.Await()
                OnPlayerResetKillStreak(Player)

    # -------- KillStreak Core Behavior --------
    OnAIElimination(Agent : ?agent) : void =
        if(PlayerAgent := Agent?):
            KillStreakCountUpdate(PlayerAgent)

    # ------------------------------------------
    KillStreakCountUpdate(Agent : agent) : void =
        #Print(PlayerNameToMessage(Agent), ?Color := NamedColors.Green, ?Duration := 10.0)
        if(Player := player[Agent], PlayerState:=PlayerPersistableData.GetPlayerState[Player]):
            #Prestige Upgrade
            PrestigeLevel:=PlayerState.PrestigeLevel
            StreakCountClamp:=MaxKillStreak + (PrestigeLevel * PrestigeIncreaseMaxKS)
            PrestigeMultiplierClampAdder:= (PrestigeLevel * PrestigeIncreaseMaxMulti)
            PrestigeMultiplierAdderByKill:= (PrestigeLevel * PrestigeAdditionalMulti)
            #End Prestige Upgrade

            Print("Kill Streak Clamp: {StreakCountClamp}")

            if:
                EnableKillClamp?
                CurrentKillCount:=PlayerTempData.GetPlayerKillStreakCount[Player]
                CurrentKillCount >= StreakCountClamp
            then:
                return
                
            NewKillCount := PlayerTempData.UpdatePlayerKillStreakCount(Player)
            CurrentMultiplier:float= (1.0 + (1.0 * KillStreakMultiplier * NewKillCount)) + PrestigeMultiplierAdderByKill
            MultiplierClamp:= MaxKillStreakMultiplier + PrestigeMultiplierClampAdder

            var FixedMultiplier:float=CurrentMultiplier

            if(EnableKillClamp?):
                set FixedMultiplier=Clamp(CurrentMultiplier, 1.0, MultiplierClamp)
                Print("Kill Streak Multiplier Clamp: {MultiplierClamp}, Clamped Value: {FixedMultiplier}")
            else:
                Print("Kill Streak Multiplier : {FixedMultiplier}")

            #Set Current Multiplier to be read from Enemies_Manager Class
            PlayerTempData.SetPlayerKillStreakMultiplier(Player, FixedMultiplier)

       
    # ------------------------------------------
    OnCharacterDamaged(DamageResult : damage_result):void=
        Target := DamageResult.Target
        Amount := DamageResult.Amount


    # ------------------------------------------
    OnPlayerResetKillStreak(Player : player):void= {}


    # --------------------------------------------------------------------------------
    InitializeUI():void=
        for (Player : GetPlayspace().GetPlayers()):
            InitializeUIForPlayer(Player)

    # --------------------------------------------------------------------------------
    InitializeUIForPlayer(Player:player):void= {}
        #NewUI := MakeKillStreakTracker(Player)
        #if (set MaybeTrackerPerPlayer[Player] = option{NewUI}):
        #    NewUI.Initialize()

    # --------------------------------------------------------------------------------
    OnPlayerJoined(Player:player):void=
        InitializeUIForPlayer(Player)