using { /Fortnite.com/Devices }
using { /Verse.org/Simulation }
using { /UnrealEngine.com/Temporary/Diagnostics }

test_room := class(creative_device):

    @editable # Stage used for enemies in this room. 0 if lobby. Use a different stage for each room
    Index<protected>:int = 0

    @editable
    RewardComponent:room_reward_device = room_reward_device{}

    @editable
    Button:button_device = button_device{}

    OnBegin<override>()<suspends>:void=
        #RewardComponent.Initialize(Index)
        Button.InteractedWithEvent.Subscribe(OnInteractedWith)

    OnInteractedWith(Agent:agent):void=
        if (Player := agent[Agent]):
            RewardComponent.GrantReward(Player)
