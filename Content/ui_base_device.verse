using { /Fortnite.com/Devices }
using { /Fortnite.com/Characters }
using { /Verse.org/Simulation }
using { /UnrealEngine.com/Temporary/Diagnostics }
using { /UnrealEngine.com/Temporary/UI }

log_ui_base_device := class(log_channel){}

<# 
Base UI for any full window shown to the player
#>
ui_base_device := class(creative_device):
    Logger:log = log{Channel:=log_ui_base_device}

    @editable # Not layer 0 so the fortnite UI like inventory is respected
    UIZOrder:type {_X:int where 0 < _X, _X <= 2147483647} = 1

    @editable
    PlayerControllerManager:player_controller_manager = player_controller_manager{}

    @editable
    RemoveInvulnerableDelay:float = 2.0

    HideEvent:event(player) = event(player){}

    NoCurrencyEvent:event(player) = event(player){}

    ShowUI(Player:player): void =
        spawn {AwaitShow(Player)}

    AwaitShow(Player:player)<suspends>: void =
        if (Player, PlayerUI:=GetPlayerUI[Player]):

            Logger.Print("Showing UI base")

            SetPlayerInUI(Player, true)

            Canvas:= ConstructCanvas(Player)
            PlayerUI.AddWidget(Canvas, player_ui_slot{ZOrder:=UIZOrder, InputMode:=ui_input_mode.All})

            AwaitShowing(Player)

            Logger.Print("Hiding UI base")

            PlayerUI.RemoveWidget(Canvas)

            SetPlayerInUI(Player, false)

    SetPlayerInUI<protected>(Player:player, bState:logic): void =
        if (PlayerController:=PlayerControllerManager.GetPlayerController[Player]):
            if (bState?):
                PlayerController.MakeInvulnerable(true)
            else:
                spawn{PlayerController.DelayMakeInvulnerable(false, RemoveInvulnerableDelay)}

    # Override with your own design
    ConstructCanvas<protected>(Player:player): canvas =
        Canvas:=canvas:
            Slots:=array:
                canvas_slot:
                    Widget:=overlay{}

    AwaitShowing<protected>(Player:player)<suspends>: void =
        AwaitHideEvent(Player)

    AwaitHideEvent<protected>(Player:player)<suspends>: void =
        if (FortCharacter := Player.GetFortCharacter[]):
            # Hide when player press hide or when eliminated
            # This prevents error of UI still running and being removed by fortnite when Manual Respawn
            race:
                FortCharacter.EliminatedEvent().Await()
                
                loop:
                    PlayerEvent := HideEvent.Await()
                    if (Player = PlayerEvent):
                        break