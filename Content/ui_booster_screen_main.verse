using { /Fortnite.com/Devices }
using { /Fortnite.com/UI}
using { /Verse.org/Simulation }
using { /Verse.org/Colors }
using { /Verse.org/Colors/NamedColors }
using { /UnrealEngine.com/Temporary/UI }
using { /UnrealEngine.com/Temporary/SpatialMath }

ui_booster_screen_main := class(creative_device):#class(ui_common)

    OnBegin<override>()<suspends>:void=
        # TODO: Replace this with your code
        CreateCanvas()
    
    CreateCanvas():void=
        Players:=GetPlayspace().GetPlayers()
        if(Player:=Players[0]):
            ImageBackground:=texture_block{DefaultImage:=Textures.General.T_Background}
            TitleTextWidget:=text_block{DefaultText := StringToMessage("BOOSTER Rewards"), DefaultTextColor := NamedColors.White}
            BackButton:button_loud = button_loud{DefaultText:=StringToMessage("<")}

            #Definition for the header
            HeaderStackBox:=overlay:
                Slots:=array:
                    overlay_slot:
                        VerticalAlignment:=vertical_alignment.Center
                        HorizontalAlignment:=horizontal_alignment.Center
                        Padding:=margin{Left:=0.0, Top:=0.0, Right:=50.0, Bottom:=0.0}
                        #Acts as a frame background to format the header area
                        Widget:=texture_block{DefaultImage:=Textures.General.T_GenSquBlueTrans, DefaultDesiredSize:=vector2{X:=1800.0, Y:=200.0}, DefaultTint:=Green}

                    overlay_slot:
                        VerticalAlignment:=vertical_alignment.Center
                        HorizontalAlignment:=horizontal_alignment.Fill
                        Padding:=margin{Left:=0.0, Top:=0.0, Right:=0.0, Bottom:=0.0}
                        Widget:=stack_box:
                            Orientation:=orientation.Horizontal
                            Slots:=array:
                                # Header's horizontal Stack Box slot
                                stack_box_slot:
                                    VerticalAlignment:=vertical_alignment.Fill
                                    HorizontalAlignment:=horizontal_alignment.Center
                                    Padding:=margin{Left:=150.0, Top:=0.0, Right:=250.0, Bottom:=0.0}
                                    #Empty space meant to keep the area for coin$ and diamonds
                                    Widget:=overlay:
                                        Slots:=array:
                                            #Makeshift size box for currency area
                                            overlay_slot:
                                                VerticalAlignment:=vertical_alignment.Center
                                                HorizontalAlignment:=horizontal_alignment.Center
                                                Padding:=margin{Left:=0.0, Top:=0.0, Right:=0.0, Bottom:=0.0}
                                                Widget:=texture_block{DefaultImage:=Textures.General.T_GenSquTrans, DefaultDesiredSize:=vector2{X:=240.0, Y:=130.0}}#}DefaultTint:=Red

                                            #Vertical stack box that will hold both currency frames
                                            overlay_slot:
                                                VerticalAlignment:=vertical_alignment.Top
                                                HorizontalAlignment:=horizontal_alignment.Left
                                                Padding:=margin{Left:=0.0, Top:=0.0, Right:=0.0, Bottom:=0.0}
                                                Widget:=stack_box:
                                                    Orientation:=orientation.Vertical
                                                    Slots:=array:
                                                        stack_box_slot:
                                                            VerticalAlignment:=vertical_alignment.Top
                                                            HorizontalAlignment:=horizontal_alignment.Center
                                                            #Distribution:=option{0.5}
                                                            Padding:=margin{Left:=0.0, Top:=0.0, Right:=0.0, Bottom:=0.0}
                                                            Widget:=overlay:
                                                                Slots:=array:
                                                                    #Coin$ frame
                                                                    overlay_slot:
                                                                        VerticalAlignment:=vertical_alignment.Fill
                                                                        HorizontalAlignment:=horizontal_alignment.Left
                                                                        Padding:=margin{Left:=0.0, Top:=0.0, Right:=0.0, Bottom:=0.0}
                                                                        Widget:=texture_block{DefaultImage:=Textures.Currency.T_GoldSlot, DefaultDesiredSize:=vector2{X:=240.0, Y:=60.0}}
                                                                    #Coin$ value
                                                                    overlay_slot:
                                                                        VerticalAlignment:=vertical_alignment.Center
                                                                        HorizontalAlignment:=horizontal_alignment.Left
                                                                        Padding:=margin{Left:=0.0, Top:=0.0, Right:=0.0, Bottom:=0.0}
                                                                        Widget:=stack_box:
                                                                            # Horizontal stack box for the currency icon and value
                                                                            Orientation:=orientation.Horizontal
                                                                            Slots:=array:
                                                                                # Coin$ icon
                                                                                stack_box_slot:
                                                                                    VerticalAlignment:=vertical_alignment.Center
                                                                                    HorizontalAlignment:=horizontal_alignment.Center
                                                                                    Padding:=margin{Left:=5.0, Top:=0.0, Right:=0.0, Bottom:=0.0}
                                                                                    Widget:=texture_block{DefaultImage:=Textures.Currency.T_Gold, DefaultDesiredSize:=vector2{X:=50.0, Y:=50.0}}
                                                                                   
                                                                                # Coin$ value    
                                                                                stack_box_slot:
                                                                                    VerticalAlignment:=vertical_alignment.Center
                                                                                    HorizontalAlignment:=horizontal_alignment.Center
                                                                                    Padding:=margin{Left:=15.0, Top:=0.0, Right:=0.0, Bottom:=0.0}
                                                                                    Widget:=text_block{DefaultText:=StringToMessage("1000"), DefaultTextColor:=NamedColors.Black}

                                                        stack_box_slot:
                                                            VerticalAlignment:=vertical_alignment.Bottom
                                                            HorizontalAlignment:=horizontal_alignment.Center
                                                            Padding:=margin{Left:=0.0, Top:=10.0, Right:=0.0, Bottom:=0.0}
                                                            #Distribution:=option{0.5}
                                                            Widget:=overlay:
                                                                Slots:=array:
                                                                    #Diamonds Frame
                                                                    overlay_slot:
                                                                        VerticalAlignment:=vertical_alignment.Fill
                                                                        HorizontalAlignment:=horizontal_alignment.Left
                                                                        Padding:=margin{Left:=0.0, Top:=0.0, Right:=0.0, Bottom:=0.0}
                                                                        Widget:=texture_block{DefaultImage:=Textures.Currency.T_DiamondSlot, DefaultDesiredSize:=vector2{X:=240.0, Y:=60.0}}

                                                                    overlay_slot:
                                                                        VerticalAlignment:=vertical_alignment.Center
                                                                        HorizontalAlignment:=horizontal_alignment.Left
                                                                        Padding:=margin{Left:=0.0, Top:=0.0, Right:=0.0, Bottom:=0.0}
                                                                        Widget:=stack_box:
                                                                            # Horizontal stack box for the currency icon and value
                                                                            Orientation:=orientation.Horizontal
                                                                            Slots:=array:
                                                                                #dIOAMONDS icon
                                                                                stack_box_slot:
                                                                                    VerticalAlignment:=vertical_alignment.Center
                                                                                    HorizontalAlignment:=horizontal_alignment.Center
                                                                                    Padding:=margin{Left:=2.5, Top:=0.0, Right:=0.0, Bottom:=0.0}
                                                                                    Widget:=texture_block{DefaultImage:=Textures.Currency.T_Diamond, DefaultDesiredSize:=vector2{X:=50.0, Y:=50.0}}

                                                                                #Diamonds value
                                                                                stack_box_slot:
                                                                                    VerticalAlignment:=vertical_alignment.Center
                                                                                    HorizontalAlignment:=horizontal_alignment.Left
                                                                                    Padding:=margin{Left:=15.0, Top:=0.0, Right:=0.0, Bottom:=0.0}
                                                                                    Widget:=text_block{DefaultText:=StringToMessage("14"), DefaultTextColor:=NamedColors.Black}


                                stack_box_slot:
                                    VerticalAlignment:=vertical_alignment.Top
                                    HorizontalAlignment:=horizontal_alignment.Center
                                    Padding:=margin{Left:=175.0, Top:=15.0, Right:=325.0, Bottom:=0.0}
                                    #Header Text
                                    Widget:=TitleTextWidget

                                stack_box_slot:
                                    VerticalAlignment:=vertical_alignment.Center
                                    HorizontalAlignment:=horizontal_alignment.Center
                                    Padding:=margin{Left:=200.0, Top:=0.0, Right:=0.0, Bottom:=0.0}
                                    #Back Button
                                    Widget:=BackButton
    
            BoosterIncreaseHealth:=overlay:
                Slots:=array:
                    #Makeshift size box, bigger than apparently needed to make room for the offset icon at the top left corner
                    overlay_slot:
                        VerticalAlignment:=vertical_alignment.Center
                        HorizontalAlignment:=horizontal_alignment.Center
                        Padding:=margin{Left:=0.0, Top:=0.0, Right:=0.0, Bottom:=0.0}
                        Widget:=texture_block{DefaultImage:=Textures.General.T_GenSquTrans, DefaultDesiredSize:=vector2{X:=1175.0, Y:=135.0}}#, DefaultTint:=Red
                
                    #Content for the booster row
                    overlay_slot:
                        VerticalAlignment:=vertical_alignment.Bottom
                        HorizontalAlignment:=horizontal_alignment.Center
                        Padding:=margin{Left:=0.0, Top:=0.0, Right:=0.0, Bottom:=0.0}
                        #Horizontal stack box
                        Widget:=stack_box:
                            Orientation:=orientation.Horizontal
                            Slots:=array:

                                # Booster icon with frame
                                stack_box_slot:
                                    VerticalAlignment:=vertical_alignment.Center
                                    HorizontalAlignment:=horizontal_alignment.Center
                                    Padding:=margin{Left:=0.0, Top:=0.0, Right:=0.0, Bottom:=0.0}
                                    Widget:=overlay:
                                        Slots:=array:
                                            #First overlay slot, with frame border
                                            overlay_slot:
                                                VerticalAlignment:=vertical_alignment.Center
                                                HorizontalAlignment:=horizontal_alignment.Center
                                                Padding:=margin{Left:=0.0, Top:=0.0, Right:=0.0, Bottom:=0.0}
                                                #Border texture
                                                Widget:=texture_block{DefaultImage:=Textures.General.T_BlueBorder, DefaultDesiredSize:=vector2{X:=120.0, Y:=120.0}}

                                            #second overlay slot, holding the item icon
                                            overlay_slot:
                                                VerticalAlignment:=vertical_alignment.Center
                                                HorizontalAlignment:=horizontal_alignment.Center
                                                Padding:=margin{Left:=0.0, Top:=0.0, Right:=0.0, Bottom:=0.0}
                                                #Item icon
                                                Widget:=texture_block{DefaultImage:=Textures.Boosters.T_IncreaseMaxHealt, DefaultDesiredSize:=vector2{X:=100.0, Y:=100.0}}
                                
                                # Booster description
                                stack_box_slot:
                                    VerticalAlignment:=vertical_alignment.Center
                                    HorizontalAlignment:=horizontal_alignment.Center
                                    Padding:=margin{Left:=0.0, Top:=0.0, Right:=0.0, Bottom:=0.0}
                                    Widget:=overlay:
                                        Slots:=array:

                                            #First overlay slot, with frame border
                                            overlay_slot:
                                                VerticalAlignment:=vertical_alignment.Center
                                                HorizontalAlignment:=horizontal_alignment.Center
                                                Padding:=margin{Left:=0.0, Top:=0.0, Right:=0.0, Bottom:=0.0}
                                                #Border texture
                                                Widget:=texture_block{DefaultImage:=Textures.General.T_BlueBorder, DefaultDesiredSize:=vector2{X:=550.0, Y:=120.0}}

                                            #Second overlay slot, description itself
                                            overlay_slot:
                                                VerticalAlignment:=vertical_alignment.Center
                                                HorizontalAlignment:=horizontal_alignment.Center
                                                Padding:=margin{Left:=0.0, Top:=0.0, Right:=0.0, Bottom:=0.0}
                                                #Border texture
                                                Widget:=stack_box:
                                                    Orientation:=orientation.Vertical
                                                    Slots:=array:
                                                        stack_box_slot:
                                                            VerticalAlignment:=vertical_alignment.Center
                                                            HorizontalAlignment:=horizontal_alignment.Center
                                                            Padding:=margin{Left:=0.0, Top:=0.0, Right:=0.0, Bottom:=0.0}
                                                            #Description header
                                                            Widget:=text_block{DefaultText:=StringToMessage("Increase Max Health"), DefaultTextColor:=NamedColors.RoyalBlue}

                                                        stack_box_slot:
                                                            VerticalAlignment:=vertical_alignment.Center
                                                            HorizontalAlignment:=horizontal_alignment.Center
                                                            Padding:=margin{Left:=0.0, Top:=0.0, Right:=0.0, Bottom:=0.0}
                                                            #Description itself
                                                            Widget:=text_block{DefaultText:=StringToMessage("+25 Max Health until you die"), DefaultTextColor:=NamedColors.White}

                                # Booster rarity description
                                stack_box_slot:
                                    VerticalAlignment:=vertical_alignment.Center
                                    HorizontalAlignment:=horizontal_alignment.Center
                                    Padding:=margin{Left:=0.0, Top:=0.0, Right:=0.0, Bottom:=0.0}
                                    Widget:=overlay:
                                        Slots:=array:

                                            #First overlay slot, with frame border
                                            overlay_slot:
                                                VerticalAlignment:=vertical_alignment.Center
                                                HorizontalAlignment:=horizontal_alignment.Center
                                                Padding:=margin{Left:=0.0, Top:=0.0, Right:=0.0, Bottom:=0.0}
                                                #Border texture
                                                Widget:=texture_block{DefaultImage:=Textures.General.T_BlueBorder, DefaultDesiredSize:=vector2{X:=550.0, Y:=120.0}}

                                            #Second overlay slot, description itself
                                            overlay_slot:
                                                VerticalAlignment:=vertical_alignment.Center
                                                HorizontalAlignment:=horizontal_alignment.Center
                                                Padding:=margin{Left:=0.0, Top:=0.0, Right:=0.0, Bottom:=0.0}
                                                #Border texture
                                                Widget:=stack_box:
                                                    Orientation:=orientation.Vertical
                                                    Slots:=array:
                                                        stack_box_slot:
                                                            VerticalAlignment:=vertical_alignment.Center
                                                            HorizontalAlignment:=horizontal_alignment.Center
                                                            Padding:=margin{Left:=0.0, Top:=0.0, Right:=0.0, Bottom:=0.0}
                                                            #Description header
                                                            Widget:=text_block{DefaultText:=StringToMessage("Next Lvl 1"), DefaultTextColor:=NamedColors.White}

                                                        stack_box_slot:
                                                            VerticalAlignment:=vertical_alignment.Center
                                                            HorizontalAlignment:=horizontal_alignment.Center
                                                            Padding:=margin{Left:=0.0, Top:=0.0, Right:=0.0, Bottom:=0.0}
                                                            #Description itself
                                                            Widget:=text_block{DefaultText:=StringToMessage("+50 Max Health until you die"), DefaultTextColor:=NamedColors.White}

                    overlay_slot:
                        VerticalAlignment:=vertical_alignment.Top
                        HorizontalAlignment:=horizontal_alignment.Left
                        Padding:=margin{Left:=0.0, Top:=0.0, Right:=0.0, Bottom:=0.0}
                        #Current Level Icon
                        Widget:=overlay:
                            Slots:=array:
                                overlay_slot:
                                    VerticalAlignment:=vertical_alignment.Center
                                    HorizontalAlignment:=horizontal_alignment.Center
                                    Padding:=margin{Left:=0.0, Top:=0.0, Right:=0.0, Bottom:=0.0}
                                    #Circular frame, current Lvl
                                    Widget:=texture_block{DefaultImage:=Textures.General.T_GenericCircle, DefaultDesiredSize:=vector2{X:=70.0, Y:=70.0}, DefaultTint:=NamedColors.Yellow}

                                overlay_slot:
                                    VerticalAlignment:=vertical_alignment.Center
                                    HorizontalAlignment:=horizontal_alignment.Center
                                    Padding:=margin{Left:=0.0, Top:=0.0, Right:=0.0, Bottom:=0.0}
                                    #Current level text block
                                    Widget:=text_block{DefaultText:=StringToMessage("Lvl 1"), DefaultTextColor:=NamedColors.Black}


            BoosterIncreaseMaxShield:=overlay:
                Slots:=array:
                    #Makeshift size box, bigger than apparently needed to make room for the offset icon at the top left corner
                    overlay_slot:
                        VerticalAlignment:=vertical_alignment.Center
                        HorizontalAlignment:=horizontal_alignment.Center
                        Padding:=margin{Left:=0.0, Top:=0.0, Right:=0.0, Bottom:=0.0}
                        Widget:=texture_block{DefaultImage:=Textures.General.T_GenSquTrans, DefaultDesiredSize:=vector2{X:=1175.0, Y:=135.0}}#, DefaultTint:=Red
                
                    #Content for the booster row
                    overlay_slot:
                        VerticalAlignment:=vertical_alignment.Bottom
                        HorizontalAlignment:=horizontal_alignment.Center
                        Padding:=margin{Left:=0.0, Top:=0.0, Right:=0.0, Bottom:=0.0}
                        #Horizontal stack box
                        Widget:=stack_box:
                            Orientation:=orientation.Horizontal
                            Slots:=array:

                                # Booster icon with frame
                                stack_box_slot:
                                    VerticalAlignment:=vertical_alignment.Center
                                    HorizontalAlignment:=horizontal_alignment.Center
                                    Padding:=margin{Left:=0.0, Top:=0.0, Right:=0.0, Bottom:=0.0}
                                    Widget:=overlay:
                                        Slots:=array:
                                            #First overlay slot, with frame border
                                            overlay_slot:
                                                VerticalAlignment:=vertical_alignment.Center
                                                HorizontalAlignment:=horizontal_alignment.Center
                                                Padding:=margin{Left:=0.0, Top:=0.0, Right:=0.0, Bottom:=0.0}
                                                #Border texture
                                                Widget:=texture_block{DefaultImage:=Textures.General.T_BlueBorder, DefaultDesiredSize:=vector2{X:=120.0, Y:=120.0}}

                                            #second overlay slot, holding the item icon
                                            overlay_slot:
                                                VerticalAlignment:=vertical_alignment.Center
                                                HorizontalAlignment:=horizontal_alignment.Center
                                                Padding:=margin{Left:=0.0, Top:=0.0, Right:=0.0, Bottom:=0.0}
                                                #Item icon
                                                Widget:=texture_block{DefaultImage:=Textures.Boosters.T_IncreaseMaxShield, DefaultDesiredSize:=vector2{X:=100.0, Y:=100.0}}
                                
                                # Booster description
                                stack_box_slot:
                                    VerticalAlignment:=vertical_alignment.Center
                                    HorizontalAlignment:=horizontal_alignment.Center
                                    Padding:=margin{Left:=0.0, Top:=0.0, Right:=0.0, Bottom:=0.0}
                                    Widget:=overlay:
                                        Slots:=array:

                                            #First overlay slot, with frame border
                                            overlay_slot:
                                                VerticalAlignment:=vertical_alignment.Center
                                                HorizontalAlignment:=horizontal_alignment.Center
                                                Padding:=margin{Left:=0.0, Top:=0.0, Right:=0.0, Bottom:=0.0}
                                                #Border texture
                                                Widget:=texture_block{DefaultImage:=Textures.General.T_BlueBorder, DefaultDesiredSize:=vector2{X:=550.0, Y:=120.0}}

                                            #Second overlay slot, description itself
                                            overlay_slot:
                                                VerticalAlignment:=vertical_alignment.Center
                                                HorizontalAlignment:=horizontal_alignment.Center
                                                Padding:=margin{Left:=0.0, Top:=0.0, Right:=0.0, Bottom:=0.0}
                                                #Border texture
                                                Widget:=stack_box:
                                                    Orientation:=orientation.Vertical
                                                    Slots:=array:
                                                        stack_box_slot:
                                                            VerticalAlignment:=vertical_alignment.Center
                                                            HorizontalAlignment:=horizontal_alignment.Center
                                                            Padding:=margin{Left:=0.0, Top:=0.0, Right:=0.0, Bottom:=0.0}
                                                            #Description header
                                                            Widget:=text_block{DefaultText:=StringToMessage("Increase Max Shield"), DefaultTextColor:=NamedColors.RoyalBlue}

                                                        stack_box_slot:
                                                            VerticalAlignment:=vertical_alignment.Center
                                                            HorizontalAlignment:=horizontal_alignment.Center
                                                            Padding:=margin{Left:=0.0, Top:=0.0, Right:=0.0, Bottom:=0.0}
                                                            #Description itself
                                                            Widget:=text_block{DefaultText:=StringToMessage("+25 Max Shield until you die"), DefaultTextColor:=NamedColors.White}

                                # Booster rarity description
                                stack_box_slot:
                                    VerticalAlignment:=vertical_alignment.Center
                                    HorizontalAlignment:=horizontal_alignment.Center
                                    Padding:=margin{Left:=0.0, Top:=0.0, Right:=0.0, Bottom:=0.0}
                                    Widget:=overlay:
                                        Slots:=array:

                                            #First overlay slot, with frame border
                                            overlay_slot:
                                                VerticalAlignment:=vertical_alignment.Center
                                                HorizontalAlignment:=horizontal_alignment.Center
                                                Padding:=margin{Left:=0.0, Top:=0.0, Right:=0.0, Bottom:=0.0}
                                                #Border texture
                                                Widget:=texture_block{DefaultImage:=Textures.General.T_BlueBorder, DefaultDesiredSize:=vector2{X:=550.0, Y:=120.0}}

                                            #Second overlay slot, description itself
                                            overlay_slot:
                                                VerticalAlignment:=vertical_alignment.Center
                                                HorizontalAlignment:=horizontal_alignment.Center
                                                Padding:=margin{Left:=0.0, Top:=0.0, Right:=0.0, Bottom:=0.0}
                                                #Border texture
                                                Widget:=stack_box:
                                                    Orientation:=orientation.Vertical
                                                    Slots:=array:
                                                        stack_box_slot:
                                                            VerticalAlignment:=vertical_alignment.Center
                                                            HorizontalAlignment:=horizontal_alignment.Center
                                                            Padding:=margin{Left:=0.0, Top:=0.0, Right:=0.0, Bottom:=0.0}
                                                            #Description header
                                                            Widget:=text_block{DefaultText:=StringToMessage("Next Lvl 1"), DefaultTextColor:=NamedColors.White}

                                                        stack_box_slot:
                                                            VerticalAlignment:=vertical_alignment.Center
                                                            HorizontalAlignment:=horizontal_alignment.Center
                                                            Padding:=margin{Left:=0.0, Top:=0.0, Right:=0.0, Bottom:=0.0}
                                                            #Description itself
                                                            Widget:=text_block{DefaultText:=StringToMessage("+50 Max Shield until you die"), DefaultTextColor:=NamedColors.White}

                    overlay_slot:
                        VerticalAlignment:=vertical_alignment.Top
                        HorizontalAlignment:=horizontal_alignment.Left
                        Padding:=margin{Left:=0.0, Top:=0.0, Right:=0.0, Bottom:=0.0}
                        #Current Level Icon
                        Widget:=overlay:
                            Slots:=array:
                                overlay_slot:
                                    VerticalAlignment:=vertical_alignment.Center
                                    HorizontalAlignment:=horizontal_alignment.Center
                                    Padding:=margin{Left:=0.0, Top:=0.0, Right:=0.0, Bottom:=0.0}
                                    #Circular frame, current Lvl
                                    Widget:=texture_block{DefaultImage:=Textures.General.T_GenericCircle, DefaultDesiredSize:=vector2{X:=70.0, Y:=70.0}, DefaultTint:=NamedColors.Yellow}

                                overlay_slot:
                                    VerticalAlignment:=vertical_alignment.Center
                                    HorizontalAlignment:=horizontal_alignment.Center
                                    Padding:=margin{Left:=0.0, Top:=0.0, Right:=0.0, Bottom:=0.0}
                                    #Current level text block
                                    Widget:=text_block{DefaultText:=StringToMessage("Lvl 1"), DefaultTextColor:=NamedColors.Black}

            BoosterIncreaseHealthRegen:=overlay:
                Slots:=array:
                    #Makeshift size box, bigger than apparently needed to make room for the offset icon at the top left corner
                    overlay_slot:
                        VerticalAlignment:=vertical_alignment.Center
                        HorizontalAlignment:=horizontal_alignment.Center
                        Padding:=margin{Left:=0.0, Top:=0.0, Right:=0.0, Bottom:=0.0}
                        Widget:=texture_block{DefaultImage:=Textures.General.T_GenSquTrans, DefaultDesiredSize:=vector2{X:=1175.0, Y:=135.0}}#, DefaultTint:=Red
                
                    #Content for the booster row
                    overlay_slot:
                        VerticalAlignment:=vertical_alignment.Bottom
                        HorizontalAlignment:=horizontal_alignment.Center
                        Padding:=margin{Left:=0.0, Top:=0.0, Right:=0.0, Bottom:=0.0}
                        #Horizontal stack box
                        Widget:=stack_box:
                            Orientation:=orientation.Horizontal
                            Slots:=array:

                                # Booster icon with frame
                                stack_box_slot:
                                    VerticalAlignment:=vertical_alignment.Center
                                    HorizontalAlignment:=horizontal_alignment.Center
                                    Padding:=margin{Left:=0.0, Top:=0.0, Right:=0.0, Bottom:=0.0}
                                    Widget:=overlay:
                                        Slots:=array:
                                            #First overlay slot, with frame border
                                            overlay_slot:
                                                VerticalAlignment:=vertical_alignment.Center
                                                HorizontalAlignment:=horizontal_alignment.Center
                                                Padding:=margin{Left:=0.0, Top:=0.0, Right:=0.0, Bottom:=0.0}
                                                #Border texture
                                                Widget:=texture_block{DefaultImage:=Textures.General.T_BlueBorder, DefaultDesiredSize:=vector2{X:=120.0, Y:=120.0}}

                                            #second overlay slot, holding the item icon
                                            overlay_slot:
                                                VerticalAlignment:=vertical_alignment.Center
                                                HorizontalAlignment:=horizontal_alignment.Center
                                                Padding:=margin{Left:=0.0, Top:=0.0, Right:=0.0, Bottom:=0.0}
                                                #Item icon
                                                Widget:=texture_block{DefaultImage:=Textures.Boosters.T_IncreaseHealthRegen, DefaultDesiredSize:=vector2{X:=100.0, Y:=100.0}}
                                
                                # Booster description
                                stack_box_slot:
                                    VerticalAlignment:=vertical_alignment.Center
                                    HorizontalAlignment:=horizontal_alignment.Center
                                    Padding:=margin{Left:=0.0, Top:=0.0, Right:=0.0, Bottom:=0.0}
                                    Widget:=overlay:
                                        Slots:=array:

                                            #First overlay slot, with frame border
                                            overlay_slot:
                                                VerticalAlignment:=vertical_alignment.Center
                                                HorizontalAlignment:=horizontal_alignment.Center
                                                Padding:=margin{Left:=0.0, Top:=0.0, Right:=0.0, Bottom:=0.0}
                                                #Border texture
                                                Widget:=texture_block{DefaultImage:=Textures.General.T_BlueBorder, DefaultDesiredSize:=vector2{X:=550.0, Y:=120.0}}

                                            #Second overlay slot, description itself
                                            overlay_slot:
                                                VerticalAlignment:=vertical_alignment.Center
                                                HorizontalAlignment:=horizontal_alignment.Center
                                                Padding:=margin{Left:=0.0, Top:=0.0, Right:=0.0, Bottom:=0.0}
                                                #Border texture
                                                Widget:=stack_box:
                                                    Orientation:=orientation.Vertical
                                                    Slots:=array:
                                                        stack_box_slot:
                                                            VerticalAlignment:=vertical_alignment.Center
                                                            HorizontalAlignment:=horizontal_alignment.Center
                                                            Padding:=margin{Left:=0.0, Top:=0.0, Right:=0.0, Bottom:=0.0}
                                                            #Description header
                                                            Widget:=text_block{DefaultText:=StringToMessage("Increase Health Regen"), DefaultTextColor:=NamedColors.RoyalBlue}

                                                        stack_box_slot:
                                                            VerticalAlignment:=vertical_alignment.Center
                                                            HorizontalAlignment:=horizontal_alignment.Center
                                                            Padding:=margin{Left:=0.0, Top:=0.0, Right:=0.0, Bottom:=0.0}
                                                            #Description itself
                                                            Widget:=text_block{DefaultText:=StringToMessage("+0.5 Health Regen"), DefaultTextColor:=NamedColors.White}

                                # Booster rarity description
                                stack_box_slot:
                                    VerticalAlignment:=vertical_alignment.Center
                                    HorizontalAlignment:=horizontal_alignment.Center
                                    Padding:=margin{Left:=0.0, Top:=0.0, Right:=0.0, Bottom:=0.0}
                                    Widget:=overlay:
                                        Slots:=array:

                                            #First overlay slot, with frame border
                                            overlay_slot:
                                                VerticalAlignment:=vertical_alignment.Center
                                                HorizontalAlignment:=horizontal_alignment.Center
                                                Padding:=margin{Left:=0.0, Top:=0.0, Right:=0.0, Bottom:=0.0}
                                                #Border texture
                                                Widget:=texture_block{DefaultImage:=Textures.General.T_BlueBorder, DefaultDesiredSize:=vector2{X:=550.0, Y:=120.0}}

                                            #Second overlay slot, description itself
                                            overlay_slot:
                                                VerticalAlignment:=vertical_alignment.Center
                                                HorizontalAlignment:=horizontal_alignment.Center
                                                Padding:=margin{Left:=0.0, Top:=0.0, Right:=0.0, Bottom:=0.0}
                                                #Border texture
                                                Widget:=stack_box:
                                                    Orientation:=orientation.Vertical
                                                    Slots:=array:
                                                        stack_box_slot:
                                                            VerticalAlignment:=vertical_alignment.Center
                                                            HorizontalAlignment:=horizontal_alignment.Center
                                                            Padding:=margin{Left:=0.0, Top:=0.0, Right:=0.0, Bottom:=0.0}
                                                            #Description header
                                                            Widget:=text_block{DefaultText:=StringToMessage("Next Lvl 1"), DefaultTextColor:=NamedColors.White}

                                                        stack_box_slot:
                                                            VerticalAlignment:=vertical_alignment.Center
                                                            HorizontalAlignment:=horizontal_alignment.Center
                                                            Padding:=margin{Left:=0.0, Top:=0.0, Right:=0.0, Bottom:=0.0}
                                                            #Description itself
                                                            Widget:=text_block{DefaultText:=StringToMessage("+1 Health Regen"), DefaultTextColor:=NamedColors.White}

                    overlay_slot:
                        VerticalAlignment:=vertical_alignment.Top
                        HorizontalAlignment:=horizontal_alignment.Left
                        Padding:=margin{Left:=0.0, Top:=0.0, Right:=0.0, Bottom:=0.0}
                        #Current Level Icon
                        Widget:=overlay:
                            Slots:=array:
                                overlay_slot:
                                    VerticalAlignment:=vertical_alignment.Center
                                    HorizontalAlignment:=horizontal_alignment.Center
                                    Padding:=margin{Left:=0.0, Top:=0.0, Right:=0.0, Bottom:=0.0}
                                    #Circular frame, current Lvl
                                    Widget:=texture_block{DefaultImage:=Textures.General.T_GenericCircle, DefaultDesiredSize:=vector2{X:=70.0, Y:=70.0}, DefaultTint:=NamedColors.Yellow}

                                overlay_slot:
                                    VerticalAlignment:=vertical_alignment.Center
                                    HorizontalAlignment:=horizontal_alignment.Center
                                    Padding:=margin{Left:=0.0, Top:=0.0, Right:=0.0, Bottom:=0.0}
                                    #Current level text block
                                    Widget:=text_block{DefaultText:=StringToMessage("Lvl 1"), DefaultTextColor:=NamedColors.Black}

            BoosterIncreaseShieldRegen:=overlay:
                Slots:=array:
                    #Makeshift size box, bigger than apparently needed to make room for the offset icon at the top left corner
                    overlay_slot:
                        VerticalAlignment:=vertical_alignment.Center
                        HorizontalAlignment:=horizontal_alignment.Center
                        Padding:=margin{Left:=0.0, Top:=0.0, Right:=0.0, Bottom:=0.0}
                        Widget:=texture_block{DefaultImage:=Textures.General.T_GenSquTrans, DefaultDesiredSize:=vector2{X:=1175.0, Y:=135.0}}#, DefaultTint:=Red
                
                    #Content for the booster row
                    overlay_slot:
                        VerticalAlignment:=vertical_alignment.Bottom
                        HorizontalAlignment:=horizontal_alignment.Center
                        Padding:=margin{Left:=0.0, Top:=0.0, Right:=0.0, Bottom:=0.0}
                        #Horizontal stack box
                        Widget:=stack_box:
                            Orientation:=orientation.Horizontal
                            Slots:=array:

                                # Booster icon with frame
                                stack_box_slot:
                                    VerticalAlignment:=vertical_alignment.Center
                                    HorizontalAlignment:=horizontal_alignment.Center
                                    Padding:=margin{Left:=0.0, Top:=0.0, Right:=0.0, Bottom:=0.0}
                                    Widget:=overlay:
                                        Slots:=array:
                                            #First overlay slot, with frame border
                                            overlay_slot:
                                                VerticalAlignment:=vertical_alignment.Center
                                                HorizontalAlignment:=horizontal_alignment.Center
                                                Padding:=margin{Left:=0.0, Top:=0.0, Right:=0.0, Bottom:=0.0}
                                                #Border texture
                                                Widget:=texture_block{DefaultImage:=Textures.General.T_BlueBorder, DefaultDesiredSize:=vector2{X:=120.0, Y:=120.0}}

                                            #second overlay slot, holding the item icon
                                            overlay_slot:
                                                VerticalAlignment:=vertical_alignment.Center
                                                HorizontalAlignment:=horizontal_alignment.Center
                                                Padding:=margin{Left:=0.0, Top:=0.0, Right:=0.0, Bottom:=0.0}
                                                #Item icon
                                                Widget:=texture_block{DefaultImage:=Textures.Boosters.T_IncreaseShieldRegen, DefaultDesiredSize:=vector2{X:=100.0, Y:=100.0}}
                                
                                # Booster description
                                stack_box_slot:
                                    VerticalAlignment:=vertical_alignment.Center
                                    HorizontalAlignment:=horizontal_alignment.Center
                                    Padding:=margin{Left:=0.0, Top:=0.0, Right:=0.0, Bottom:=0.0}
                                    Widget:=overlay:
                                        Slots:=array:

                                            #First overlay slot, with frame border
                                            overlay_slot:
                                                VerticalAlignment:=vertical_alignment.Center
                                                HorizontalAlignment:=horizontal_alignment.Center
                                                Padding:=margin{Left:=0.0, Top:=0.0, Right:=0.0, Bottom:=0.0}
                                                #Border texture
                                                Widget:=texture_block{DefaultImage:=Textures.General.T_BlueBorder, DefaultDesiredSize:=vector2{X:=550.0, Y:=120.0}}

                                            #Second overlay slot, description itself
                                            overlay_slot:
                                                VerticalAlignment:=vertical_alignment.Center
                                                HorizontalAlignment:=horizontal_alignment.Center
                                                Padding:=margin{Left:=0.0, Top:=0.0, Right:=0.0, Bottom:=0.0}
                                                #Border texture
                                                Widget:=stack_box:
                                                    Orientation:=orientation.Vertical
                                                    Slots:=array:
                                                        stack_box_slot:
                                                            VerticalAlignment:=vertical_alignment.Center
                                                            HorizontalAlignment:=horizontal_alignment.Center
                                                            Padding:=margin{Left:=0.0, Top:=0.0, Right:=0.0, Bottom:=0.0}
                                                            #Description header
                                                            Widget:=text_block{DefaultText:=StringToMessage("Increase Shield Regen"), DefaultTextColor:=NamedColors.RoyalBlue}

                                                        stack_box_slot:
                                                            VerticalAlignment:=vertical_alignment.Center
                                                            HorizontalAlignment:=horizontal_alignment.Center
                                                            Padding:=margin{Left:=0.0, Top:=0.0, Right:=0.0, Bottom:=0.0}
                                                            #Description itself
                                                            Widget:=text_block{DefaultText:=StringToMessage("+0.5 Shield Regen"), DefaultTextColor:=NamedColors.White}

                                # Booster rarity description
                                stack_box_slot:
                                    VerticalAlignment:=vertical_alignment.Center
                                    HorizontalAlignment:=horizontal_alignment.Center
                                    Padding:=margin{Left:=0.0, Top:=0.0, Right:=0.0, Bottom:=0.0}
                                    Widget:=overlay:
                                        Slots:=array:

                                            #First overlay slot, with frame border
                                            overlay_slot:
                                                VerticalAlignment:=vertical_alignment.Center
                                                HorizontalAlignment:=horizontal_alignment.Center
                                                Padding:=margin{Left:=0.0, Top:=0.0, Right:=0.0, Bottom:=0.0}
                                                #Border texture
                                                Widget:=texture_block{DefaultImage:=Textures.General.T_BlueBorder, DefaultDesiredSize:=vector2{X:=550.0, Y:=120.0}}

                                            #Second overlay slot, description itself
                                            overlay_slot:
                                                VerticalAlignment:=vertical_alignment.Center
                                                HorizontalAlignment:=horizontal_alignment.Center
                                                Padding:=margin{Left:=0.0, Top:=0.0, Right:=0.0, Bottom:=0.0}
                                                #Border texture
                                                Widget:=stack_box:
                                                    Orientation:=orientation.Vertical
                                                    Slots:=array:
                                                        stack_box_slot:
                                                            VerticalAlignment:=vertical_alignment.Center
                                                            HorizontalAlignment:=horizontal_alignment.Center
                                                            Padding:=margin{Left:=0.0, Top:=0.0, Right:=0.0, Bottom:=0.0}
                                                            #Description header
                                                            Widget:=text_block{DefaultText:=StringToMessage("Next Lvl 1"), DefaultTextColor:=NamedColors.White}

                                                        stack_box_slot:
                                                            VerticalAlignment:=vertical_alignment.Center
                                                            HorizontalAlignment:=horizontal_alignment.Center
                                                            Padding:=margin{Left:=0.0, Top:=0.0, Right:=0.0, Bottom:=0.0}
                                                            #Description itself
                                                            Widget:=text_block{DefaultText:=StringToMessage("+1 shield Regen"), DefaultTextColor:=NamedColors.White}

                    overlay_slot:
                        VerticalAlignment:=vertical_alignment.Top
                        HorizontalAlignment:=horizontal_alignment.Left
                        Padding:=margin{Left:=0.0, Top:=0.0, Right:=0.0, Bottom:=0.0}
                        #Current Level Icon
                        Widget:=overlay:
                            Slots:=array:
                                overlay_slot:
                                    VerticalAlignment:=vertical_alignment.Center
                                    HorizontalAlignment:=horizontal_alignment.Center
                                    Padding:=margin{Left:=0.0, Top:=0.0, Right:=0.0, Bottom:=0.0}
                                    #Circular frame, current Lvl
                                    Widget:=texture_block{DefaultImage:=Textures.General.T_GenericCircle, DefaultDesiredSize:=vector2{X:=70.0, Y:=70.0}, DefaultTint:=NamedColors.Yellow}

                                overlay_slot:
                                    VerticalAlignment:=vertical_alignment.Center
                                    HorizontalAlignment:=horizontal_alignment.Center
                                    Padding:=margin{Left:=0.0, Top:=0.0, Right:=0.0, Bottom:=0.0}
                                    #Current level text block
                                    Widget:=text_block{DefaultText:=StringToMessage("Lvl 1"), DefaultTextColor:=NamedColors.Black}

            BoosterVampireBullets:=overlay:
                Slots:=array:
                    #Makeshift size box, bigger than apparently needed to make room for the offset icon at the top left corner
                    overlay_slot:
                        VerticalAlignment:=vertical_alignment.Center
                        HorizontalAlignment:=horizontal_alignment.Center
                        Padding:=margin{Left:=0.0, Top:=0.0, Right:=0.0, Bottom:=0.0}
                        Widget:=texture_block{DefaultImage:=Textures.General.T_GenSquTrans, DefaultDesiredSize:=vector2{X:=1175.0, Y:=135.0}}#, DefaultTint:=Red
                
                    #Content for the booster row
                    overlay_slot:
                        VerticalAlignment:=vertical_alignment.Bottom
                        HorizontalAlignment:=horizontal_alignment.Center
                        Padding:=margin{Left:=0.0, Top:=0.0, Right:=0.0, Bottom:=0.0}
                        #Horizontal stack box
                        Widget:=stack_box:
                            Orientation:=orientation.Horizontal
                            Slots:=array:

                                # Booster icon with frame
                                stack_box_slot:
                                    VerticalAlignment:=vertical_alignment.Center
                                    HorizontalAlignment:=horizontal_alignment.Center
                                    Padding:=margin{Left:=0.0, Top:=0.0, Right:=0.0, Bottom:=0.0}
                                    Widget:=overlay:
                                        Slots:=array:
                                            #First overlay slot, with frame border
                                            overlay_slot:
                                                VerticalAlignment:=vertical_alignment.Center
                                                HorizontalAlignment:=horizontal_alignment.Center
                                                Padding:=margin{Left:=0.0, Top:=0.0, Right:=0.0, Bottom:=0.0}
                                                #Border texture
                                                Widget:=texture_block{DefaultImage:=Textures.General.T_BlueBorder, DefaultDesiredSize:=vector2{X:=120.0, Y:=120.0}}

                                            #second overlay slot, holding the item icon
                                            overlay_slot:
                                                VerticalAlignment:=vertical_alignment.Center
                                                HorizontalAlignment:=horizontal_alignment.Center
                                                Padding:=margin{Left:=0.0, Top:=0.0, Right:=0.0, Bottom:=0.0}
                                                #Item icon
                                                Widget:=texture_block{DefaultImage:=Textures.Boosters.T_Vampire, DefaultDesiredSize:=vector2{X:=100.0, Y:=100.0}}
                                
                                # Booster description
                                stack_box_slot:
                                    VerticalAlignment:=vertical_alignment.Center
                                    HorizontalAlignment:=horizontal_alignment.Center
                                    Padding:=margin{Left:=0.0, Top:=0.0, Right:=0.0, Bottom:=0.0}
                                    Widget:=overlay:
                                        Slots:=array:

                                            #First overlay slot, with frame border
                                            overlay_slot:
                                                VerticalAlignment:=vertical_alignment.Center
                                                HorizontalAlignment:=horizontal_alignment.Center
                                                Padding:=margin{Left:=0.0, Top:=0.0, Right:=0.0, Bottom:=0.0}
                                                #Border texture
                                                Widget:=texture_block{DefaultImage:=Textures.General.T_BlueBorder, DefaultDesiredSize:=vector2{X:=550.0, Y:=120.0}}

                                            #Second overlay slot, description itself
                                            overlay_slot:
                                                VerticalAlignment:=vertical_alignment.Center
                                                HorizontalAlignment:=horizontal_alignment.Center
                                                Padding:=margin{Left:=0.0, Top:=0.0, Right:=0.0, Bottom:=0.0}
                                                #Border texture
                                                Widget:=stack_box:
                                                    Orientation:=orientation.Vertical
                                                    Slots:=array:
                                                        stack_box_slot:
                                                            VerticalAlignment:=vertical_alignment.Center
                                                            HorizontalAlignment:=horizontal_alignment.Center
                                                            Padding:=margin{Left:=0.0, Top:=0.0, Right:=0.0, Bottom:=0.0}
                                                            #Description header
                                                            Widget:=text_block{DefaultText:=StringToMessage("Vampire Bullets"), DefaultTextColor:=NamedColors.RoyalBlue}

                                                        stack_box_slot:
                                                            VerticalAlignment:=vertical_alignment.Center
                                                            HorizontalAlignment:=horizontal_alignment.Center
                                                            Padding:=margin{Left:=0.0, Top:=0.0, Right:=0.0, Bottom:=0.0}
                                                            #Description itself
                                                            Widget:=text_block{DefaultText:=StringToMessage("Get 5% HP from DMG inflicted\nfor 2 mins"), DefaultTextColor:=NamedColors.White}

                                # Booster rarity description
                                stack_box_slot:
                                    VerticalAlignment:=vertical_alignment.Center
                                    HorizontalAlignment:=horizontal_alignment.Center
                                    Padding:=margin{Left:=0.0, Top:=0.0, Right:=0.0, Bottom:=0.0}
                                    Widget:=overlay:
                                        Slots:=array:

                                            #First overlay slot, with frame border
                                            overlay_slot:
                                                VerticalAlignment:=vertical_alignment.Center
                                                HorizontalAlignment:=horizontal_alignment.Center
                                                Padding:=margin{Left:=0.0, Top:=0.0, Right:=0.0, Bottom:=0.0}
                                                #Border texture
                                                Widget:=texture_block{DefaultImage:=Textures.General.T_BlueBorder, DefaultDesiredSize:=vector2{X:=550.0, Y:=120.0}}

                                            #Second overlay slot, description itself
                                            overlay_slot:
                                                VerticalAlignment:=vertical_alignment.Center
                                                HorizontalAlignment:=horizontal_alignment.Center
                                                Padding:=margin{Left:=0.0, Top:=0.0, Right:=0.0, Bottom:=0.0}
                                                #Border texture
                                                Widget:=stack_box:
                                                    Orientation:=orientation.Vertical
                                                    Slots:=array:
                                                        stack_box_slot:
                                                            VerticalAlignment:=vertical_alignment.Center
                                                            HorizontalAlignment:=horizontal_alignment.Center
                                                            Padding:=margin{Left:=0.0, Top:=0.0, Right:=0.0, Bottom:=0.0}
                                                            #Description header
                                                            Widget:=text_block{DefaultText:=StringToMessage("Next Lvl 1"), DefaultTextColor:=NamedColors.White}

                                                        stack_box_slot:
                                                            VerticalAlignment:=vertical_alignment.Center
                                                            HorizontalAlignment:=horizontal_alignment.Center
                                                            Padding:=margin{Left:=0.0, Top:=0.0, Right:=0.0, Bottom:=0.0}
                                                            #Description itself
                                                            Widget:=text_block{DefaultText:=StringToMessage("Get 10% HP from DMG inflicted\nfor 2 mins"), DefaultTextColor:=NamedColors.White}

                    overlay_slot:
                        VerticalAlignment:=vertical_alignment.Top
                        HorizontalAlignment:=horizontal_alignment.Left
                        Padding:=margin{Left:=0.0, Top:=0.0, Right:=0.0, Bottom:=0.0}
                        #Current Level Icon
                        Widget:=overlay:
                            Slots:=array:
                                overlay_slot:
                                    VerticalAlignment:=vertical_alignment.Center
                                    HorizontalAlignment:=horizontal_alignment.Center
                                    Padding:=margin{Left:=0.0, Top:=0.0, Right:=0.0, Bottom:=0.0}
                                    #Circular frame, current Lvl
                                    Widget:=texture_block{DefaultImage:=Textures.General.T_GenericCircle, DefaultDesiredSize:=vector2{X:=70.0, Y:=70.0}, DefaultTint:=NamedColors.Yellow}

                                overlay_slot:
                                    VerticalAlignment:=vertical_alignment.Center
                                    HorizontalAlignment:=horizontal_alignment.Center
                                    Padding:=margin{Left:=0.0, Top:=0.0, Right:=0.0, Bottom:=0.0}
                                    #Current level text block
                                    Widget:=text_block{DefaultText:=StringToMessage("Lvl 1"), DefaultTextColor:=NamedColors.Black}

            #Defines several components to be used along the implementation:
            HeaderCanvas:canvas = canvas:
                Slots:=array:
                    #Defines the parameters for the canvas' slot
                    canvas_slot:
                        Anchors := anchors{Minimum := vector2{X := 0.0, Y := 0.0}, Maximum := vector2{X := 1.0, Y := 1.0}}
                        Alignment := vector2{X := 0.5, Y := 0.5}
                        Widget:=overlay:
                            Slots:=array:
                                overlay_slot:
                                    VerticalAlignment:=vertical_alignment.Center
                                    HorizontalAlignment:=horizontal_alignment.Center
                                    Padding:=margin{Left:=0.0, Top:=0.0, Right:=0.0, Bottom:=0.0}
                                    #Screen background
                                    Widget:=texture_block{DefaultImage:=Textures.General.T_Background, DefaultDesiredSize:=vector2{X:=1910.0, Y:=1080.0}}

                                overlay_slot:
                                    VerticalAlignment:=vertical_alignment.Top
                                    HorizontalAlignment:=horizontal_alignment.Center
                                    Padding:=margin{Left:=0.0, Top:=0.0, Right:=0.0, Bottom:=0.0}
                                    #Content
                                    Widget:=stack_box:
                                        Orientation:=orientation.Vertical
                                        Slots:=array:
                                            stack_box_slot:
                                                VerticalAlignment:=vertical_alignment.Top
                                                HorizontalAlignment:=horizontal_alignment.Center
                                                Padding:=margin{Left:=0.0, Top:=0.0, Right:=0.0, Bottom:=0.0}
                                                #Header
                                                Widget:=HeaderStackBox

                                            stack_box_slot:
                                                VerticalAlignment:=vertical_alignment.Top
                                                HorizontalAlignment:=horizontal_alignment.Center
                                                Padding:=margin{Left:=0.0, Top:=20.0, Right:=0.0, Bottom:=0.0}
                                                # Increase Max Health
                                                Widget:=BoosterIncreaseHealth

                                            stack_box_slot:
                                                VerticalAlignment:=vertical_alignment.Top
                                                HorizontalAlignment:=horizontal_alignment.Center
                                                Padding:=margin{Left:=0.0, Top:=0.0, Right:=0.0, Bottom:=0.0}
                                                # Increase Max Health
                                                Widget:=BoosterIncreaseMaxShield

                                            stack_box_slot:
                                                VerticalAlignment:=vertical_alignment.Top
                                                HorizontalAlignment:=horizontal_alignment.Center
                                                Padding:=margin{Left:=0.0, Top:=0.0, Right:=0.0, Bottom:=0.0}
                                                # Increase Health Regen
                                                Widget:=BoosterIncreaseHealthRegen

                                            stack_box_slot:
                                                VerticalAlignment:=vertical_alignment.Top
                                                HorizontalAlignment:=horizontal_alignment.Center
                                                Padding:=margin{Left:=0.0, Top:=0.0, Right:=0.0, Bottom:=0.0}
                                                # Increase Health Regen
                                                Widget:=BoosterIncreaseShieldRegen

                                            stack_box_slot:
                                                VerticalAlignment:=vertical_alignment.Top
                                                HorizontalAlignment:=horizontal_alignment.Center
                                                Padding:=margin{Left:=0.0, Top:=0.0, Right:=0.0, Bottom:=0.0}
                                                # Increase Health Regen
                                                Widget:=BoosterVampireBullets
                                            

            if(PlayerUI:=GetPlayerUI[Player]):
                PlayerUI.AddWidget(HeaderCanvas)