using { /Fortnite.com/Devices }
using { /Fortnite.com/UI}
using { /Verse.org/Simulation }
using { /Verse.org/Colors }
using { /Verse.org/Colors/NamedColors }
using { /UnrealEngine.com/Temporary/UI }
using { /UnrealEngine.com/Temporary/SpatialMath }

ui_booster_upgrade_device := class(creative_device):#class(ui_common)

    OnBegin<override>()<suspends>:void=
        CreateCanvas()
    
    CreateCanvas():void=
        Players:=GetPlayspace().GetPlayers()
        if(Player:=Players[0]):
            ImageBackground:=texture_block{DefaultImage:=Textures.General.T_Background}
            TitleTextWidget:=text_block{DefaultText := StringToMessage("BOOSTER Rewards"), DefaultTextColor := NamedColors.White}
            BackButton:button_loud = button_loud{DefaultText:=StringToMessage("CLOSE")}

            #Definition for the header
            HeaderStackBox:=overlay:
                Slots:=array:
                    overlay_slot:
                        VerticalAlignment:=vertical_alignment.Center
                        HorizontalAlignment:=horizontal_alignment.Center
                        Padding:=margin{Left:=0.0, Top:=0.0, Right:=50.0, Bottom:=0.0}
                        #Acts as a frame background to format the header area
                        Widget:=texture_block{DefaultImage:=Textures.General.T_GenSquBlueTrans, DefaultDesiredSize:=vector2{X:=1800.0, Y:=200.0}, DefaultTint:=Green}

                    overlay_slot:
                        VerticalAlignment:=vertical_alignment.Center
                        HorizontalAlignment:=horizontal_alignment.Fill
                        Padding:=margin{Left:=0.0, Top:=0.0, Right:=0.0, Bottom:=0.0}
                        Widget:=stack_box:
                            Orientation:=orientation.Horizontal
                            Slots:=array:
                                # Header's horizontal Stack Box slot
                                stack_box_slot:
                                    VerticalAlignment:=vertical_alignment.Center
                                    HorizontalAlignment:=horizontal_alignment.Center
                                    Padding:=margin{Left:=100.0, Top:=0.0, Right:=200.0, Bottom:=0.0}
                                    #Empty space meant to keep the area for coin$ and diamonds
                                    Widget:=overlay:
                                        Slots:=array:
                                            #Makeshift size box for currency area
                                            overlay_slot:
                                                VerticalAlignment:=vertical_alignment.Center
                                                HorizontalAlignment:=horizontal_alignment.Center
                                                Padding:=margin{Left:=0.0, Top:=0.0, Right:=0.0, Bottom:=0.0}
                                                Widget:=texture_block{DefaultImage:=Textures.General.T_GenSquTrans, DefaultDesiredSize:=vector2{X:=490.0, Y:=60.0}, DefaultTint:=Red}#}DefaultTint:=Red

                                            #Horizontal stack box that will hold both currency frames
                                            overlay_slot:
                                                VerticalAlignment:=vertical_alignment.Top
                                                HorizontalAlignment:=horizontal_alignment.Left
                                                Padding:=margin{Left:=0.0, Top:=0.0, Right:=0.0, Bottom:=0.0}
                                                Widget:=stack_box:
                                                    Orientation:=orientation.Horizontal
                                                    Slots:=array:
                                                        stack_box_slot:
                                                            VerticalAlignment:=vertical_alignment.Center
                                                            HorizontalAlignment:=horizontal_alignment.Left
                                                            #Distribution:=option{0.5}
                                                            Padding:=margin{Left:=0.0, Top:=0.0, Right:=0.0, Bottom:=0.0}
                                                            Widget:=overlay:
                                                                Slots:=array:
                                                                    #Coin$ frame
                                                                    overlay_slot:
                                                                        VerticalAlignment:=vertical_alignment.Fill
                                                                        HorizontalAlignment:=horizontal_alignment.Left
                                                                        Padding:=margin{Left:=0.0, Top:=0.0, Right:=0.0, Bottom:=0.0}
                                                                        Widget:=texture_block{DefaultImage:=Textures.Currency.T_GoldSlot, DefaultDesiredSize:=vector2{X:=240.0, Y:=60.0}}
                                                                    #Coin$ value
                                                                    overlay_slot:
                                                                        VerticalAlignment:=vertical_alignment.Center
                                                                        HorizontalAlignment:=horizontal_alignment.Left
                                                                        Padding:=margin{Left:=0.0, Top:=0.0, Right:=0.0, Bottom:=0.0}
                                                                        Widget:=stack_box:
                                                                            # Horizontal stack box for the currency icon and value
                                                                            Orientation:=orientation.Horizontal
                                                                            Slots:=array:
                                                                                # Coin$ icon
                                                                                stack_box_slot:
                                                                                    VerticalAlignment:=vertical_alignment.Center
                                                                                    HorizontalAlignment:=horizontal_alignment.Center
                                                                                    Padding:=margin{Left:=5.0, Top:=0.0, Right:=0.0, Bottom:=0.0}
                                                                                    Widget:=texture_block{DefaultImage:=Textures.Currency.T_Gold, DefaultDesiredSize:=vector2{X:=50.0, Y:=50.0}}
                                                                                   
                                                                                # Coin$ value    
                                                                                stack_box_slot:
                                                                                    VerticalAlignment:=vertical_alignment.Center
                                                                                    HorizontalAlignment:=horizontal_alignment.Center
                                                                                    Padding:=margin{Left:=15.0, Top:=0.0, Right:=0.0, Bottom:=0.0}
                                                                                    Widget:=text_block{DefaultText:=StringToMessage("1000"), DefaultTextColor:=NamedColors.Black}

                                                        stack_box_slot:
                                                            VerticalAlignment:=vertical_alignment.Center
                                                            HorizontalAlignment:=horizontal_alignment.Right
                                                            Padding:=margin{Left:=10.0, Top:=0.0, Right:=0.0, Bottom:=0.0}
                                                            #Distribution:=option{0.5}
                                                            Widget:=overlay:
                                                                Slots:=array:
                                                                    #Diamonds Frame
                                                                    overlay_slot:
                                                                        VerticalAlignment:=vertical_alignment.Fill
                                                                        HorizontalAlignment:=horizontal_alignment.Left
                                                                        Padding:=margin{Left:=0.0, Top:=0.0, Right:=0.0, Bottom:=0.0}
                                                                        Widget:=texture_block{DefaultImage:=Textures.Currency.T_DiamondSlot, DefaultDesiredSize:=vector2{X:=240.0, Y:=60.0}}

                                                                    overlay_slot:
                                                                        VerticalAlignment:=vertical_alignment.Center
                                                                        HorizontalAlignment:=horizontal_alignment.Left
                                                                        Padding:=margin{Left:=0.0, Top:=0.0, Right:=0.0, Bottom:=0.0}
                                                                        Widget:=stack_box:
                                                                            # Horizontal stack box for the currency icon and value
                                                                            Orientation:=orientation.Horizontal
                                                                            Slots:=array:
                                                                                #dIOAMONDS icon
                                                                                stack_box_slot:
                                                                                    VerticalAlignment:=vertical_alignment.Center
                                                                                    HorizontalAlignment:=horizontal_alignment.Center
                                                                                    Padding:=margin{Left:=2.5, Top:=0.0, Right:=0.0, Bottom:=0.0}
                                                                                    Widget:=texture_block{DefaultImage:=Textures.Currency.T_Diamond, DefaultDesiredSize:=vector2{X:=50.0, Y:=50.0}}

                                                                                #Diamonds value
                                                                                stack_box_slot:
                                                                                    VerticalAlignment:=vertical_alignment.Center
                                                                                    HorizontalAlignment:=horizontal_alignment.Left
                                                                                    Padding:=margin{Left:=15.0, Top:=0.0, Right:=0.0, Bottom:=0.0}
                                                                                    Widget:=text_block{DefaultText:=StringToMessage("14"), DefaultTextColor:=NamedColors.Black}


                                stack_box_slot:
                                    VerticalAlignment:=vertical_alignment.Top
                                    HorizontalAlignment:=horizontal_alignment.Center
                                    Padding:=margin{Left:=105.0, Top:=5.0, Right:=330.0, Bottom:=0.0}
                                    #Header Text
                                    Widget:=TitleTextWidget

                                stack_box_slot:
                                    VerticalAlignment:=vertical_alignment.Center
                                    HorizontalAlignment:=horizontal_alignment.Right
                                    Padding:=margin{Left:=110.0, Top:=0.0, Right:=0.0, Bottom:=0.0}
                                    #Back Button
                                    Widget:=BackButton

                                    #Definition for the default level rows
            LevelRow0:=overlay:
                Slots:=array:
                    overlay_slot:
                        #Makeshift size box
                        VerticalAlignment:=vertical_alignment.Center
                        HorizontalAlignment:=horizontal_alignment.Center
                        Padding:=margin{Left:=0.0, Top:=0.0, Right:=0.0, Bottom:=0.0}
                        Widget:=texture_block{DefaultImage:=Textures.General.T_GenSquTrans, DefaultDesiredSize:=vector2{X:=850.0, Y:=130.0}, DefaultTint:=MakeColorFromHex("#0000FFFF")}

                    overlay_slot:
                        #Rarity level content
                        VerticalAlignment:=vertical_alignment.Center
                        HorizontalAlignment:=horizontal_alignment.Left
                        Padding:=margin{Left:=0.0, Top:=0.0, Right:=0.0, Bottom:=0.0}
                        Widget:=overlay:
                            Slots:=array:

                                overlay_slot:
                                    #Makeshift size box
                                    VerticalAlignment:=vertical_alignment.Center
                                    HorizontalAlignment:=horizontal_alignment.Center
                                    Padding:=margin{Left:=0.0, Top:=0.0, Right:=0.0, Bottom:=0.0}
                                    Widget:=texture_block{DefaultImage:=Textures.General.T_GenericSquare, DefaultDesiredSize:=vector2{X:=850.0, Y:=130.0}, DefaultTint:=MakeColorFromHex("#010a1832")}

                                overlay_slot:
                                    #Highlight texture, over the content but under the rarity flag
                                    VerticalAlignment:=vertical_alignment.Fill
                                    HorizontalAlignment:=horizontal_alignment.Fill
                                    Padding:=margin{Left:=0.0, Top:=0.0, Right:=0.0, Bottom:=0.0}
                                    Widget:=texture_block{DefaultImage:=Textures.General.T_LevelBorder, DefaultDesiredSize:=vector2{X:=512.0, Y:=64.0}, DefaultTint:=MakeColorFromHex("#1d4b89FF")}

                                    #Horizontal stack box
                                overlay_slot:
                                    VerticalAlignment:=vertical_alignment.Fill
                                    HorizontalAlignment:=horizontal_alignment.Left
                                    Padding:=margin{Left:=0.0, Top:=0.0, Right:=0.0, Bottom:=0.0}
                                    Widget:=stack_box:
                                        Orientation:=orientation.Horizontal
                                        Slots:=array:
                                            #Level flag
                                            stack_box_slot:
                                                VerticalAlignment:=vertical_alignment.Top
                                                HorizontalAlignment:=horizontal_alignment.Left
                                                Padding:=margin{Left:=0.0, Top:=0.0, Right:=15.0, Bottom:=0.0}
                                                Widget:=overlay:

                                                    Slots:=array:
                                                        overlay_slot:
                                                            VerticalAlignment:=vertical_alignment.Center
                                                            HorizontalAlignment:=horizontal_alignment.Center
                                                            Padding:=margin{Left:=0.0, Top:=0.0, Right:=0.0, Bottom:=0.0}
                                                            #Tier level makeshift size box
                                                            Widget:=texture_block{DefaultImage:=Textures.General.T_GenSquTrans, DefaultDesiredSize:=vector2{X:=150.0, Y:=130.0}, DefaultTint:=MakeColorFromHex("#FF0000FF")}

                                                        overlay_slot:
                                                            VerticalAlignment:=vertical_alignment.Top
                                                            HorizontalAlignment:=horizontal_alignment.Left
                                                            Padding:=margin{Left:=0.0, Top:=0.0, Right:=0.0, Bottom:=0.0}
                                                            #Selected level yellow flag
                                                            Widget:=stack_box:
                                                                Orientation:=orientation.Vertical
                                                                Slots:=array:
                                                                    stack_box_slot:
                                                                        VerticalAlignment:=vertical_alignment.Center
                                                                        HorizontalAlignment:=horizontal_alignment.Center
                                                                        Padding:=margin{Left:=0.0, Top:=0.0, Right:=0.0, Bottom:=0.0}
                                                                        #level flag texture
                                                                        Widget:=overlay:
                                                                            Slots:=array:

                                                                                overlay_slot:
                                                                                    VerticalAlignment:=vertical_alignment.Center
                                                                                    HorizontalAlignment:=horizontal_alignment.Center
                                                                                    Padding:=margin{Left:=0.0, Top:=0.0, Right:=0.0, Bottom:=0.0}
                                                                                    #level flag texture
                                                                                    Widget:=texture_block{DefaultImage:=Textures.General.T_GenLevelFlag, DefaultDesiredSize:=vector2{X:=150.0, Y:=45.0}, DefaultTint:=MakeColorFromHex("#f7ff1aFF")}

                                                                                overlay_slot:
                                                                                    VerticalAlignment:=vertical_alignment.Center
                                                                                    HorizontalAlignment:=horizontal_alignment.Center
                                                                                    Padding:=margin{Left:=0.0, Top:=0.0, Right:=0.0, Bottom:=0.0}
                                                                                    #level flag text
                                                                                    Widget:=text_block{DefaultText := StringToMessage("Base LV"), DefaultTextColor:=MakeColorFromHex("#000000FF")}

                                                                    stack_box_slot:
                                                                        VerticalAlignment:=vertical_alignment.Center
                                                                        HorizontalAlignment:=horizontal_alignment.Center
                                                                        Padding:=margin{Left:=0.0, Top:=0.0, Right:=0.0, Bottom:=0.0}
                                                                        #level flag texture
                                                                        Widget:=overlay:
                                                                            Slots:=array:

                                                                                #Booster Icon - Increase Max Shield
                                                                                overlay_slot:
                                                                                    VerticalAlignment:=vertical_alignment.Center
                                                                                    HorizontalAlignment:=horizontal_alignment.Center
                                                                                    Padding:=margin{Left:=0.0, Top:=0.0, Right:=0.0, Bottom:=0.0}
                                                                                    #Weapon icon BG
                                                                                    Widget:=texture_block{DefaultImage:=Textures.Rarity.T_IconBG_Common, DefaultDesiredSize:=vector2{X:=78.0, Y:=78.0}, DefaultTint:=MakeColorFromHex("#FFFFFFFF")}
                        
                                                                                overlay_slot:
                                                                                    VerticalAlignment:=vertical_alignment.Center
                                                                                    HorizontalAlignment:=horizontal_alignment.Center
                                                                                    Padding:=margin{Left:=0.0, Top:=0.0, Right:=0.0, Bottom:=0.0}
                                                                                    #Weapon icon Border
                                                                                    Widget:=texture_block{DefaultImage:=Textures.General.T_BoosterFrame, DefaultDesiredSize:=vector2{X:=78.0, Y:=78.0}, DefaultTint:=MakeColorFromHex("#FFFFFFFF")}
                        
                                                                                overlay_slot:
                                                                                    VerticalAlignment:=vertical_alignment.Center
                                                                                    HorizontalAlignment:=horizontal_alignment.Center
                                                                                    Padding:=margin{Left:=0.0, Top:=0.0, Right:=0.0, Bottom:=0.0}
                                                                                    #Weapon icon - pistol
                                                                                    Widget:=texture_block{DefaultImage:=Textures.Boosters.T_IncreaseMaxShield, DefaultDesiredSize:=vector2{X:=70.0, Y:=70.0}, DefaultTint:=MakeColorFromHex("#FFFFFFFF")}
                        
                                                                                #overlay_slot:
                                                                                    #VerticalAlignment:=vertical_alignment.Bottom
                                                                                    #HorizontalAlignment:=horizontal_alignment.Center
                                                                                    #Padding:=margin{Left:=0.0, Top:=0.0, Right:=0.0, Bottom:=5.0}
                                                                                    ##Weapon header - pistol
                                                                                    #Widget:=texture_block{DefaultImage:=Textures.Items.T_HeaderPistol, DefaultDesiredSize:=vector2{X:=78.0, Y:=18.0}, DefaultTint:=MakeColorFromHex("#FFFFFFFF")}

                                                        #overlay_slot:
                                                            #VerticalAlignment:=vertical_alignment.Fill
                                                            #HorizontalAlignment:=horizontal_alignment.Center
                                                            #Padding:=margin{Left:=0.0, Top:=5.0, Right:=0.0, Bottom:=5.0}
                                                            #Unselected level and lock
                                                            #Widget:=overlay:
                                                                #Slots:=array:
                                                                    
                                                                    #overlay_slot:
                                                                        #VerticalAlignment:=vertical_alignment.Top
                                                                        #HorizontalAlignment:=horizontal_alignment.Center
                                                                        #Padding:=margin{Left:=0.0, Top:=0.0, Right:=0.0, Bottom:=0.0}
                                                                        #level flag texture
                                                                        #Widget:=text_block{DefaultText := StringToMessage("Base Level"), DefaultTextColor:=MakeColorFromHex("#FFFFFFFF")}

                                                                    #overlay_slot:
                                                                        #VerticalAlignment:=vertical_alignment.Bottom
                                                                        #HorizontalAlignment:=horizontal_alignment.Center
                                                                        #Padding:=margin{Left:=0.0, Top:=0.0, Right:=0.0, Bottom:=0.0}
                                                                        ##level flag texture
                                                                        #Widget:=overlay:
                                                                            #Slots:=array:
                                                                                ##Weapon Icon - Pistols
                                                                                #overlay_slot:
                                                                                    #VerticalAlignment:=vertical_alignment.Center
                                                                                    #HorizontalAlignment:=horizontal_alignment.Center
                                                                                    #Padding:=margin{Left:=0.0, Top:=0.0, Right:=0.0, Bottom:=0.0}
                                                                                    ##Weapon icon BG
                                                                                    #Widget:=texture_block{DefaultImage:=Textures.General.T_GenericSquare, DefaultDesiredSize:=vector2{X:=78.0, Y:=78.0}, DefaultTint:=MakeColorFromHex("#000c1d88")}
                        
                                                                                #overlay_slot:
                                                                                    #VerticalAlignment:=vertical_alignment.Center
                                                                                    #HorizontalAlignment:=horizontal_alignment.Center
                                                                                    #Padding:=margin{Left:=0.0, Top:=0.0, Right:=0.0, Bottom:=0.0}
                                                                                    ##Weapon icon Border
                                                                                    #Widget:=texture_block{DefaultImage:=Textures.General.T_GenericBorderSquare, DefaultDesiredSize:=vector2{X:=78.0, Y:=78.0}, DefaultTint:=MakeColorFromHex("#2e7be3FF")}
                        
                                                                                #overlay_slot:
                                                                                    #VerticalAlignment:=vertical_alignment.Top
                                                                                    #HorizontalAlignment:=horizontal_alignment.Center
                                                                                    #Padding:=margin{Left:=0.0, Top:=5.0, Right:=0.0, Bottom:=0.0}
                                                                                    ##Weapon icon - pistol
                                                                                    #Widget:=texture_block{DefaultImage:=Textures.Items.T_IconPistol, DefaultDesiredSize:=vector2{X:=50.0, Y:=50.0}, DefaultTint:=MakeColorFromHex("#FFFFFFFF")}
                        
                                                                                #overlay_slot:
                                                                                    #VerticalAlignment:=vertical_alignment.Bottom
                                                                                    #HorizontalAlignment:=horizontal_alignment.Center
                                                                                    #Padding:=margin{Left:=0.0, Top:=0.0, Right:=0.0, Bottom:=5.0}
                                                                                    ##Weapon header - pistol
                                                                                    #Widget:=texture_block{DefaultImage:=Textures.Items.T_HeaderPistol, DefaultDesiredSize:=vector2{X:=78.0, Y:=18.0}, DefaultTint:=MakeColorFromHex("#FFFFFFFF")}
                                            
                                            

                                            stack_box_slot:
                                                VerticalAlignment:=vertical_alignment.Fill
                                                HorizontalAlignment:=horizontal_alignment.Center #Center
                                                Padding:=margin{Left:=0.0, Top:=0.0, Right:=0.0, Bottom:=0.0}
                                                Widget:=overlay:
                                                    Slots:=array:
                                                        overlay_slot:
                                                            #Makeshift size box
                                                            VerticalAlignment:=vertical_alignment.Center
                                                            HorizontalAlignment:=horizontal_alignment.Center
                                                            Padding:=margin{Left:=0.0, Top:=0.0, Right:=0.0, Bottom:=0.0}
                                                            Widget:=texture_block{DefaultImage:=Textures.General.T_GenSquTrans, DefaultDesiredSize:=vector2{X:=418.0, Y:=100.0}, DefaultTint:=MakeColorFromHex("#00FF0011")}

                                                        overlay_slot:
                                                            #Vertical stack box to add all the rarity information
                                                            VerticalAlignment:=vertical_alignment.Center
                                                            HorizontalAlignment:=horizontal_alignment.Center
                                                            Padding:=margin{Left:=0.0, Top:=0.0, Right:=0.0, Bottom:=0.0}
                                                            Widget:=stack_box:
                                                                Orientation:=orientation.Vertical
                                                                Slots:=array:
                                                                    stack_box_slot:
                                                                        #Horizontal Stack box for max rarity available
                                                                        VerticalAlignment:=vertical_alignment.Top
                                                                        HorizontalAlignment:=horizontal_alignment.Center
                                                                        Padding:=margin{Left:=0.0, Top:=0.0, Right:=0.0, Bottom:=0.0}
                                                                        Widget:=stack_box:
                                                                            Orientation:=orientation.Horizontal
                                                                            Slots:=array:
                                                                                stack_box_slot:
                                                                                    VerticalAlignment:=vertical_alignment.Center
                                                                                    HorizontalAlignment:=horizontal_alignment.Center
                                                                                    Padding:=margin{Left:=0.0, Top:=0.0, Right:=0.0, Bottom:=0.0}
                                                                                    Widget:=text_block{DefaultText := StringToMessage("Increase Max Shield"), DefaultTextColor:=MakeColorFromHex("#FFFFFFFF")}

                                                                    stack_box_slot:
                                                                        #Horizontal overlay for overall rarity scheme
                                                                        VerticalAlignment:=vertical_alignment.Bottom
                                                                        HorizontalAlignment:=horizontal_alignment.Center
                                                                        Padding:=margin{Left:=0.0, Top:=0.0, Right:=0.0, Bottom:=0.0}
                                                                        Widget:=overlay:
                                                                            Slots:=array:
                                                                                
                                                                                #Makeshift size box for the rarity scheme
                                                                                overlay_slot:
                                                                                    VerticalAlignment:=vertical_alignment.Center
                                                                                    HorizontalAlignment:=horizontal_alignment.Center
                                                                                    Padding:=margin{Left:=0.0, Top:=0.0, Right:=0.0, Bottom:=0.0}
                                                                                    Widget:=texture_block{DefaultImage:=Textures.General.T_GenSquTrans, DefaultDesiredSize:=vector2{X:=255.0, Y:=50.0}, DefaultTint:=MakeColorFromHex("#FF000022")}#Red frame

                                                                                #Visible border for rarity layout
                                                                                overlay_slot:
                                                                                    VerticalAlignment:=vertical_alignment.Center
                                                                                    HorizontalAlignment:=horizontal_alignment.Center
                                                                                    Padding:=margin{Left:=0.0, Top:=0.0, Right:=0.0, Bottom:=0.0}
                                                                                    Widget:=texture_block{DefaultImage:=Textures.General.T_GenericBorderSquare3, DefaultDesiredSize:=vector2{X:=255.0, Y:=40.0}, DefaultTint:=MakeColorFromHex("#2e7be3FF")}

                                                                                #Horizontal Stack box for rarity slots
                                                                                overlay_slot:
                                                                                    VerticalAlignment:=vertical_alignment.Fill
                                                                                    HorizontalAlignment:=horizontal_alignment.Fill
                                                                                    Padding:=margin{Left:=5.0, Top:=0.0, Right:=5.0, Bottom:=0.0}
                                                                                    Widget:=stack_box:
                                                                                        Orientation:=orientation.Horizontal
                                                                                        Slots:=array:
                                                                                            #Overlay for each rarity slots, including the pointing arrow
                                                                                            stack_box_slot:
                                                                                                VerticalAlignment:=vertical_alignment.Center
                                                                                                HorizontalAlignment:=horizontal_alignment.Center
                                                                                                Distribution:=option{1.0}
                                                                                                Padding:=margin{Left:=0.0, Top:=0.0, Right:=0.0, Bottom:=0.0}
                                                                                                Widget:=overlay:
                                                                                                    Slots:=array:
                                                                                                        #Makeshift size box
                                                                                                        overlay_slot:
                                                                                                            VerticalAlignment:=vertical_alignment.Center
                                                                                                            HorizontalAlignment:=horizontal_alignment.Center
                                                                                                            Padding:=margin{Left:=0.0, Top:=0.0, Right:=0.0, Bottom:=0.0}
                                                                                                            Widget:=texture_block{DefaultImage:=Textures.General.T_GenSquTrans, DefaultDesiredSize:=vector2{X:=30.0, Y:=40.0}, DefaultTint:=MakeColorFromHex("#0000FF08")}

                                                                                                        #Overlay for rarity border and texture
                                                                                                        overlay_slot:
                                                                                                            VerticalAlignment:=vertical_alignment.Center
                                                                                                            HorizontalAlignment:=horizontal_alignment.Center
                                                                                                            Padding:=margin{Left:=0.0, Top:=0.0, Right:=0.0, Bottom:=0.0}
                                                                                                            Widget:=overlay:
                                                                                                                Slots:=array:
                                                                                                                    
                                                                                                                    #Rarity texture border, when the rarity slot is empty
                                                                                                                    #overlay_slot:
                                                                                                                        #VerticalAlignment:=vertical_alignment.Center
                                                                                                                        #HorizontalAlignment:=horizontal_alignment.Center
                                                                                                                        #Padding:=margin{Left:=0.0, Top:=0.0, Right:=0.0, Bottom:=0.0}
                                                                                                                        #Widget:=texture_block{DefaultImage:=Textures.General.T_GenericBorderSquare2, DefaultDesiredSize:=vector2{X:=30.0, Y:=30.0}, DefaultTint:=MakeColorFromHex("#2e7be388")}

                                                                                                                    #Rarity texture
                                                                                                                    overlay_slot:
                                                                                                                        VerticalAlignment:=vertical_alignment.Center
                                                                                                                        HorizontalAlignment:=horizontal_alignment.Center
                                                                                                                        Padding:=margin{Left:=0.0, Top:=0.0, Right:=0.0, Bottom:=0.0}
                                                                                                                        Widget:=texture_block{DefaultImage:=Textures.General.T_GenericSquare, DefaultDesiredSize:=vector2{X:=20.0, Y:=20.0}, DefaultTint:=MakeColorFromHex("#979797FF")}
                                                                                                        
                                                                                                                    

                                                                                                        #Bracket
                                                                                                        overlay_slot:
                                                                                                            VerticalAlignment:=vertical_alignment.Fill
                                                                                                            HorizontalAlignment:=horizontal_alignment.Fill
                                                                                                            Padding:=margin{Left:=0.0, Top:=0.0, Right:=0.0, Bottom:=0.0}
                                                                                                            Widget:=stack_box:
                                                                                                                Orientation:=orientation.Horizontal
                                                                                                                #two instances of bracket texture to represent the range of rarities
                                                                                                                Slots:=array:

                                                                                                                    stack_box_slot:
                                                                                                                        VerticalAlignment:=vertical_alignment.Center
                                                                                                                        HorizontalAlignment:=horizontal_alignment.Left
                                                                                                                        Distribution:=option{1.0}
                                                                                                                        Padding:=margin{Left:=0.0, Top:=0.0, Right:=0.0, Bottom:=0.0}
                                                                                                                        Widget:=texture_block{DefaultImage:=Textures.General.T_RarityBracketLeft, DefaultDesiredSize:=vector2{X:=18.0, Y:=36.0}, DefaultTint:=MakeColorFromHex("#FFFFFF44")}

                                                                                                                    stack_box_slot:
                                                                                                                        VerticalAlignment:=vertical_alignment.Center
                                                                                                                        HorizontalAlignment:=horizontal_alignment.Right
                                                                                                                        Distribution:=option{1.0}
                                                                                                                        Padding:=margin{Left:=0.0, Top:=0.0, Right:=0.0, Bottom:=0.0}
                                                                                                                        Widget:=texture_block{DefaultImage:=Textures.General.T_RarityBracketRight, DefaultDesiredSize:=vector2{X:=18.0, Y:=36.0}, DefaultTint:=MakeColorFromHex("#FFFFFF44")}

                                                                                            stack_box_slot:
                                                                                                VerticalAlignment:=vertical_alignment.Center
                                                                                                HorizontalAlignment:=horizontal_alignment.Center
                                                                                                Distribution:=option{1.0}
                                                                                                Padding:=margin{Left:=0.0, Top:=0.0, Right:=0.0, Bottom:=0.0}
                                                                                                Widget:=overlay:
                                                                                                    Slots:=array:
                                                                                                        #Makeshift size box
                                                                                                        overlay_slot:
                                                                                                            VerticalAlignment:=vertical_alignment.Center
                                                                                                            HorizontalAlignment:=horizontal_alignment.Center
                                                                                                            Padding:=margin{Left:=0.0, Top:=0.0, Right:=0.0, Bottom:=0.0}
                                                                                                            Widget:=texture_block{DefaultImage:=Textures.General.T_GenSquTrans, DefaultDesiredSize:=vector2{X:=30.0, Y:=40.0}, DefaultTint:=MakeColorFromHex("#0000FF08")}

                                                                                                        #Overlay for rarity border and texture
                                                                                                        overlay_slot:
                                                                                                            VerticalAlignment:=vertical_alignment.Center
                                                                                                            HorizontalAlignment:=horizontal_alignment.Center
                                                                                                            Padding:=margin{Left:=0.0, Top:=0.0, Right:=0.0, Bottom:=0.0}
                                                                                                            Widget:=overlay:
                                                                                                                Slots:=array:
                                                                                                                    
                                                                                                                    #Rarity texture border, when the rarity slot is empty
                                                                                                                    overlay_slot:
                                                                                                                        VerticalAlignment:=vertical_alignment.Center
                                                                                                                        HorizontalAlignment:=horizontal_alignment.Center
                                                                                                                        Padding:=margin{Left:=0.0, Top:=0.0, Right:=0.0, Bottom:=0.0}
                                                                                                                        Widget:=texture_block{DefaultImage:=Textures.General.T_GenericBorderSquare2, DefaultDesiredSize:=vector2{X:=30.0, Y:=30.0}, DefaultTint:=MakeColorFromHex("#2e7be388")}

                                                                                                                    #Rarity texture
                                                                                                                    #overlay_slot:
                                                                                                                        #VerticalAlignment:=vertical_alignment.Center
                                                                                                                        #HorizontalAlignment:=horizontal_alignment.Center
                                                                                                                        #Padding:=margin{Left:=0.0, Top:=0.0, Right:=0.0, Bottom:=0.0}
                                                                                                                        #Widget:=texture_block{DefaultImage:=Textures.General.T_GenericSquare, DefaultDesiredSize:=vector2{X:=30.0, Y:=30.0}, DefaultTint:=MakeColorFromHex("#4ccc2aFF")}

                                                                                                        #Bracket
                                                                                                        #overlay_slot:
                                                                                                            #VerticalAlignment:=vertical_alignment.Fill
                                                                                                            #HorizontalAlignment:=horizontal_alignment.Fill
                                                                                                            #Padding:=margin{Left:=0.0, Top:=0.0, Right:=0.0, Bottom:=0.0}
                                                                                                            #Widget:=stack_box:
                                                                                                                #Orientation:=orientation.Horizontal
                                                                                                                #two instances of bracket texture to represent the range of rarities
                                                                                                                #Slots:=array:

                                                                                                                    #stack_box_slot:
                                                                                                                        #VerticalAlignment:=vertical_alignment.Center
                                                                                                                        #HorizontalAlignment:=horizontal_alignment.Left
                                                                                                                        #Distribution:=option{1.0}
                                                                                                                        #Padding:=margin{Left:=0.0, Top:=0.0, Right:=0.0, Bottom:=0.0}
                                                                                                                        #Widget:=texture_block{DefaultImage:=Textures.General.T_RarityBracketLeft, DefaultDesiredSize:=vector2{X:=18.0, Y:=36.0}, DefaultTint:=MakeColorFromHex("#FFFFFF44")}

                                                                                                                    #stack_box_slot:
                                                                                                                        #VerticalAlignment:=vertical_alignment.Center
                                                                                                                        #HorizontalAlignment:=horizontal_alignment.Right
                                                                                                                        #Distribution:=option{1.0}
                                                                                                                        #Padding:=margin{Left:=0.0, Top:=0.0, Right:=0.0, Bottom:=0.0}
                                                                                                                        #Widget:=texture_block{DefaultImage:=Textures.General.T_RarityBracketRight, DefaultDesiredSize:=vector2{X:=18.0, Y:=36.0}, DefaultTint:=MakeColorFromHex("#FFFFFF44")}

                                                                                            stack_box_slot:
                                                                                                VerticalAlignment:=vertical_alignment.Center
                                                                                                HorizontalAlignment:=horizontal_alignment.Center
                                                                                                Distribution:=option{1.0}
                                                                                                Padding:=margin{Left:=0.0, Top:=0.0, Right:=0.0, Bottom:=0.0}
                                                                                                Widget:=overlay:
                                                                                                    Slots:=array:
                                                                                                        #Makeshift size box
                                                                                                        overlay_slot:
                                                                                                            VerticalAlignment:=vertical_alignment.Center
                                                                                                            HorizontalAlignment:=horizontal_alignment.Center
                                                                                                            Padding:=margin{Left:=0.0, Top:=0.0, Right:=0.0, Bottom:=0.0}
                                                                                                            Widget:=texture_block{DefaultImage:=Textures.General.T_GenSquTrans, DefaultDesiredSize:=vector2{X:=30.0, Y:=40.0}, DefaultTint:=MakeColorFromHex("#0000FF08")}

                                                                                                        #Overlay for rarity border and texture
                                                                                                        overlay_slot:
                                                                                                            VerticalAlignment:=vertical_alignment.Center
                                                                                                            HorizontalAlignment:=horizontal_alignment.Center
                                                                                                            Padding:=margin{Left:=0.0, Top:=0.0, Right:=0.0, Bottom:=0.0}
                                                                                                            Widget:=overlay:
                                                                                                                Slots:=array:
                                                                                                                                                                                                                
                                                                                                                    #Rarity texture border, when the rarity slot is empty
                                                                                                                    overlay_slot:
                                                                                                                        VerticalAlignment:=vertical_alignment.Center
                                                                                                                        HorizontalAlignment:=horizontal_alignment.Center
                                                                                                                        Padding:=margin{Left:=0.0, Top:=0.0, Right:=0.0, Bottom:=0.0}
                                                                                                                        Widget:=texture_block{DefaultImage:=Textures.General.T_GenericBorderSquare2, DefaultDesiredSize:=vector2{X:=30.0, Y:=30.0}, DefaultTint:=MakeColorFromHex("#2e7be388")}
                                                                                                                                                                                                                            
                                                                                                                    #Rarity texture
                                                                                                                    #overlay_slot:
                                                                                                                        #VerticalAlignment:=vertical_alignment.Bottom
                                                                                                                        #HorizontalAlignment:=horizontal_alignment.Center
                                                                                                                        #Padding:=margin{Left:=0.0, Top:=0.0, Right:=0.0, Bottom:=0.0}
                                                                                                                        #Widget:=texture_block{DefaultImage:=Textures.General.T_GenericSquare, DefaultDesiredSize:=vector2{X:=30.0, Y:=30.0}, DefaultTint:=MakeColorFromHex("#2a54ccFF")}
                                                                                                                    
                                                                                                        #Bracket
                                                                                                        #overlay_slot:
                                                                                                            #VerticalAlignment:=vertical_alignment.Fill
                                                                                                            #HorizontalAlignment:=horizontal_alignment.Fill
                                                                                                            #Padding:=margin{Left:=0.0, Top:=0.0, Right:=0.0, Bottom:=0.0}
                                                                                                            #Widget:=stack_box:
                                                                                                                #Orientation:=orientation.Horizontal
                                                                                                                #two instances of bracket texture to represent the range of rarities
                                                                                                                #Slots:=array:

                                                                                                                    #stack_box_slot:
                                                                                                                        #VerticalAlignment:=vertical_alignment.Center
                                                                                                                        #HorizontalAlignment:=horizontal_alignment.Left
                                                                                                                        #Distribution:=option{1.0}
                                                                                                                        #Padding:=margin{Left:=0.0, Top:=0.0, Right:=0.0, Bottom:=0.0}
                                                                                                                        #Widget:=texture_block{DefaultImage:=Textures.General.T_RarityBracketLeft, DefaultDesiredSize:=vector2{X:=18.0, Y:=36.0}, DefaultTint:=MakeColorFromHex("#FFFFFF44")}

                                                                                                                    #stack_box_slot:
                                                                                                                        #VerticalAlignment:=vertical_alignment.Center
                                                                                                                        #HorizontalAlignment:=horizontal_alignment.Right
                                                                                                                        #Distribution:=option{1.0}
                                                                                                                        #Padding:=margin{Left:=0.0, Top:=0.0, Right:=0.0, Bottom:=0.0}
                                                                                                                        #Widget:=texture_block{DefaultImage:=Textures.General.T_RarityBracketRight, DefaultDesiredSize:=vector2{X:=18.0, Y:=36.0}, DefaultTint:=MakeColorFromHex("#FFFFFF44")}

                                                                                            stack_box_slot:
                                                                                                VerticalAlignment:=vertical_alignment.Center
                                                                                                HorizontalAlignment:=horizontal_alignment.Center
                                                                                                Distribution:=option{1.0}
                                                                                                Padding:=margin{Left:=0.0, Top:=0.0, Right:=0.0, Bottom:=0.0}
                                                                                                Widget:=overlay:
                                                                                                    Slots:=array:
                                                                                                        #Makeshift size box
                                                                                                        overlay_slot:
                                                                                                            VerticalAlignment:=vertical_alignment.Center
                                                                                                            HorizontalAlignment:=horizontal_alignment.Center
                                                                                                            Padding:=margin{Left:=0.0, Top:=0.0, Right:=0.0, Bottom:=0.0}
                                                                                                            Widget:=texture_block{DefaultImage:=Textures.General.T_GenSquTrans, DefaultDesiredSize:=vector2{X:=30.0, Y:=40.0}, DefaultTint:=MakeColorFromHex("#0000FF08")}

                                                                                                        #Overlay for rarity border and texture
                                                                                                        overlay_slot:
                                                                                                            VerticalAlignment:=vertical_alignment.Center
                                                                                                            HorizontalAlignment:=horizontal_alignment.Center
                                                                                                            Padding:=margin{Left:=0.0, Top:=0.0, Right:=0.0, Bottom:=0.0}
                                                                                                            Widget:=overlay:
                                                                                                                Slots:=array:

                                                                                                                    #Rarity texture border, when the rarity slot is empty
                                                                                                                    overlay_slot:
                                                                                                                        VerticalAlignment:=vertical_alignment.Center
                                                                                                                        HorizontalAlignment:=horizontal_alignment.Center
                                                                                                                        Padding:=margin{Left:=0.0, Top:=0.0, Right:=0.0, Bottom:=0.0}
                                                                                                                        Widget:=texture_block{DefaultImage:=Textures.General.T_GenericBorderSquare2, DefaultDesiredSize:=vector2{X:=30.0, Y:=30.0}, DefaultTint:=MakeColorFromHex("#2e7be388")}
                                                                                                                    
                                                                                                                    #Rarity texture
                                                                                                                    #overlay_slot:
                                                                                                                        #VerticalAlignment:=vertical_alignment.Bottom
                                                                                                                        #HorizontalAlignment:=horizontal_alignment.Center
                                                                                                                        #Padding:=margin{Left:=0.0, Top:=0.0, Right:=0.0, Bottom:=0.0}
                                                                                                                        #Widget:=texture_block{DefaultImage:=Textures.General.T_GenericSquare, DefaultDesiredSize:=vector2{X:=30.0, Y:=30.0}, DefaultTint:=MakeColorFromHex("#b32accFF")}
                                                                                                                    
                                                                                                        #Bracket
                                                                                                        #overlay_slot:
                                                                                                            #VerticalAlignment:=vertical_alignment.Fill
                                                                                                            #HorizontalAlignment:=horizontal_alignment.Fill
                                                                                                            #Padding:=margin{Left:=0.0, Top:=0.0, Right:=0.0, Bottom:=0.0}
                                                                                                            #Widget:=stack_box:
                                                                                                                #Orientation:=orientation.Horizontal
                                                                                                                #two instances of bracket texture to represent the range of rarities
                                                                                                                #Slots:=array:

                                                                                                                    #stack_box_slot:
                                                                                                                        #VerticalAlignment:=vertical_alignment.Center
                                                                                                                        #HorizontalAlignment:=horizontal_alignment.Left
                                                                                                                        #Distribution:=option{1.0}
                                                                                                                        #Padding:=margin{Left:=0.0, Top:=0.0, Right:=0.0, Bottom:=0.0}
                                                                                                                        #Widget:=texture_block{DefaultImage:=Textures.General.T_RarityBracketLeft, DefaultDesiredSize:=vector2{X:=18.0, Y:=36.0}, DefaultTint:=MakeColorFromHex("#FFFFFF44")}

                                                                                                                    #stack_box_slot:
                                                                                                                        #VerticalAlignment:=vertical_alignment.Center
                                                                                                                        #HorizontalAlignment:=horizontal_alignment.Right
                                                                                                                        #Distribution:=option{1.0}
                                                                                                                        #Padding:=margin{Left:=0.0, Top:=0.0, Right:=0.0, Bottom:=0.0}
                                                                                                                        #Widget:=texture_block{DefaultImage:=Textures.General.T_RarityBracketRight, DefaultDesiredSize:=vector2{X:=18.0, Y:=36.0}, DefaultTint:=MakeColorFromHex("#FFFFFF44")}

                                                                                            stack_box_slot:
                                                                                                VerticalAlignment:=vertical_alignment.Center
                                                                                                HorizontalAlignment:=horizontal_alignment.Center
                                                                                                Distribution:=option{1.0}
                                                                                                Padding:=margin{Left:=0.0, Top:=0.0, Right:=0.0, Bottom:=0.0}
                                                                                                Widget:=overlay:
                                                                                                    Slots:=array:
                                                                                                        #Makeshift size box
                                                                                                        overlay_slot:
                                                                                                            VerticalAlignment:=vertical_alignment.Center
                                                                                                            HorizontalAlignment:=horizontal_alignment.Center
                                                                                                            Padding:=margin{Left:=0.0, Top:=0.0, Right:=0.0, Bottom:=0.0}
                                                                                                            Widget:=texture_block{DefaultImage:=Textures.General.T_GenSquTrans, DefaultDesiredSize:=vector2{X:=30.0, Y:=40.0}, DefaultTint:=MakeColorFromHex("#0000FF08")}

                                                                                                        #Overlay for rarity border and texture
                                                                                                        overlay_slot:
                                                                                                            VerticalAlignment:=vertical_alignment.Center
                                                                                                            HorizontalAlignment:=horizontal_alignment.Center
                                                                                                            Padding:=margin{Left:=0.0, Top:=0.0, Right:=0.0, Bottom:=0.0}
                                                                                                            Widget:=overlay:
                                                                                                                Slots:=array:

                                                                                                                    #Rarity texture border, when the rarity slot is empty
                                                                                                                    overlay_slot:
                                                                                                                        VerticalAlignment:=vertical_alignment.Center
                                                                                                                        HorizontalAlignment:=horizontal_alignment.Center
                                                                                                                        Padding:=margin{Left:=0.0, Top:=0.0, Right:=0.0, Bottom:=0.0}
                                                                                                                        Widget:=texture_block{DefaultImage:=Textures.General.T_GenericBorderSquare2, DefaultDesiredSize:=vector2{X:=30.0, Y:=30.0}, DefaultTint:=MakeColorFromHex("#2e7be388")}
                                                                                                                    
                                                                                                                    #Rarity texture
                                                                                                                    #overlay_slot:
                                                                                                                        #VerticalAlignment:=vertical_alignment.Bottom
                                                                                                                        #HorizontalAlignment:=horizontal_alignment.Center
                                                                                                                        #Padding:=margin{Left:=0.0, Top:=0.0, Right:=0.0, Bottom:=0.0}
                                                                                                                        #Widget:=texture_block{DefaultImage:=Textures.General.T_GenericSquare, DefaultDesiredSize:=vector2{X:=30.0, Y:=30.0}, DefaultTint:=MakeColorFromHex("#cc8d2aFF")}

                                                                                                        #Bracket
                                                                                                        #overlay_slot:
                                                                                                            #VerticalAlignment:=vertical_alignment.Fill
                                                                                                            #HorizontalAlignment:=horizontal_alignment.Fill
                                                                                                            #Padding:=margin{Left:=0.0, Top:=0.0, Right:=0.0, Bottom:=0.0}
                                                                                                            #Widget:=stack_box:
                                                                                                                #Orientation:=orientation.Horizontal
                                                                                                                #two instances of bracket texture to represent the range of rarities
                                                                                                                #Slots:=array:

                                                                                                                    #stack_box_slot:
                                                                                                                        #VerticalAlignment:=vertical_alignment.Center
                                                                                                                        #HorizontalAlignment:=horizontal_alignment.Left
                                                                                                                        #Distribution:=option{1.0}
                                                                                                                        #Padding:=margin{Left:=0.0, Top:=0.0, Right:=0.0, Bottom:=0.0}
                                                                                                                        #Widget:=texture_block{DefaultImage:=Textures.General.T_RarityBracketLeft, DefaultDesiredSize:=vector2{X:=18.0, Y:=36.0}, DefaultTint:=MakeColorFromHex("#FFFFFF44")}

                                                                                                                    #stack_box_slot:
                                                                                                                        #VerticalAlignment:=vertical_alignment.Center
                                                                                                                        #HorizontalAlignment:=horizontal_alignment.Right
                                                                                                                        #Distribution:=option{1.0}
                                                                                                                        #Padding:=margin{Left:=0.0, Top:=0.0, Right:=0.0, Bottom:=0.0}
                                                                                                                        #Widget:=texture_block{DefaultImage:=Textures.General.T_RarityBracketRight, DefaultDesiredSize:=vector2{X:=18.0, Y:=36.0}, DefaultTint:=MakeColorFromHex("#FFFFFF44")}

                                                                                            stack_box_slot:
                                                                                                VerticalAlignment:=vertical_alignment.Center
                                                                                                HorizontalAlignment:=horizontal_alignment.Center
                                                                                                Distribution:=option{1.0}
                                                                                                Padding:=margin{Left:=0.0, Top:=0.0, Right:=0.0, Bottom:=0.0}
                                                                                                Widget:=overlay:
                                                                                                    Slots:=array:
                                                                                                        #Makeshift size box
                                                                                                        overlay_slot:
                                                                                                            VerticalAlignment:=vertical_alignment.Center
                                                                                                            HorizontalAlignment:=horizontal_alignment.Center
                                                                                                            Padding:=margin{Left:=0.0, Top:=0.0, Right:=0.0, Bottom:=0.0}
                                                                                                            Widget:=texture_block{DefaultImage:=Textures.General.T_GenSquTrans, DefaultDesiredSize:=vector2{X:=30.0, Y:=40.0}, DefaultTint:=MakeColorFromHex("#0000FF08")}

                                                                                                        #Overlay for rarity border and texture
                                                                                                        overlay_slot:
                                                                                                            VerticalAlignment:=vertical_alignment.Center
                                                                                                            HorizontalAlignment:=horizontal_alignment.Center
                                                                                                            Padding:=margin{Left:=0.0, Top:=0.0, Right:=0.0, Bottom:=0.0}
                                                                                                            Widget:=overlay:
                                                                                                                Slots:=array:
                                                                                                        
                                                                                                                    #Rarity texture border, when the rarity slot is empty
                                                                                                                    overlay_slot:
                                                                                                                        VerticalAlignment:=vertical_alignment.Center
                                                                                                                        HorizontalAlignment:=horizontal_alignment.Center
                                                                                                                        Padding:=margin{Left:=0.0, Top:=0.0, Right:=0.0, Bottom:=0.0}
                                                                                                                        Widget:=texture_block{DefaultImage:=Textures.General.T_GenericBorderSquare2, DefaultDesiredSize:=vector2{X:=30.0, Y:=30.0}, DefaultTint:=MakeColorFromHex("#2e7be388")}
                                                                                                                    
                                                                                                                    #Rarity texture
                                                                                                                    #overlay_slot:
                                                                                                                        #VerticalAlignment:=vertical_alignment.Bottom
                                                                                                                        #HorizontalAlignment:=horizontal_alignment.Center
                                                                                                                        #Padding:=margin{Left:=0.0, Top:=0.0, Right:=0.0, Bottom:=0.0}
                                                                                                                        #Widget:=texture_block{DefaultImage:=Textures.General.T_GenericSquare, DefaultDesiredSize:=vector2{X:=30.0, Y:=30.0}, DefaultTint:=MakeColorFromHex("#eacf19FF")}
                                                                                                                    
                                                                                                        #Bracket
                                                                                                        #overlay_slot:
                                                                                                            #VerticalAlignment:=vertical_alignment.Fill
                                                                                                            #HorizontalAlignment:=horizontal_alignment.Fill
                                                                                                            #Padding:=margin{Left:=0.0, Top:=0.0, Right:=0.0, Bottom:=0.0}
                                                                                                            #Widget:=stack_box:
                                                                                                                #Orientation:=orientation.Horizontal
                                                                                                                #two instances of bracket texture to represent the range of rarities
                                                                                                                #Slots:=array:

                                                                                                                    #stack_box_slot:
                                                                                                                        #VerticalAlignment:=vertical_alignment.Center
                                                                                                                        #HorizontalAlignment:=horizontal_alignment.Left
                                                                                                                        #Distribution:=option{1.0}
                                                                                                                        #Padding:=margin{Left:=0.0, Top:=0.0, Right:=0.0, Bottom:=0.0}
                                                                                                                        #Widget:=texture_block{DefaultImage:=Textures.General.T_RarityBracketLeft, DefaultDesiredSize:=vector2{X:=18.0, Y:=36.0}, DefaultTint:=MakeColorFromHex("#FFFFFF44")}

                                                                                                                    #stack_box_slot:
                                                                                                                        #VerticalAlignment:=vertical_alignment.Center
                                                                                                                        #HorizontalAlignment:=horizontal_alignment.Right
                                                                                                                        #Distribution:=option{1.0}
                                                                                                                        #Padding:=margin{Left:=0.0, Top:=0.0, Right:=0.0, Bottom:=0.0}
                                                                                                                        #Widget:=texture_block{DefaultImage:=Textures.General.T_RarityBracketRight, DefaultDesiredSize:=vector2{X:=18.0, Y:=36.0}, DefaultTint:=MakeColorFromHex("#FFFFFF44")}
                                            #Upgrade button
                                            stack_box_slot:
                                                VerticalAlignment:=vertical_alignment.Center
                                                HorizontalAlignment:=horizontal_alignment.Center
                                                Padding:=margin{Left:=0.0, Top:=0.0, Right:=0.0, Bottom:=0.0}
                                                Widget:=button_loud{DefaultText:=StringToMessage("UPDATE\n$15")}

                    #overlay_slot:
                        #VerticalAlignment:=vertical_alignment.Fill
                        #HorizontalAlignment:=horizontal_alignment.Fill
                        #Padding:=margin{Left:=0.0, Top:=0.0, Right:=0.0, Bottom:=0.0}
                        #Widget:=texture_block{DefaultImage:=Textures.General.T_LevelPointer, DefaultDesiredSize:=vector2{X:=20.0, Y:=20.0}, DefaultTint:=MakeColorFromHex("#7de2FFFF")}
                 
            LevelRow1:=overlay:
                Slots:=array:
                    overlay_slot:
                        #Makeshift size box
                        VerticalAlignment:=vertical_alignment.Center
                        HorizontalAlignment:=horizontal_alignment.Center
                        Padding:=margin{Left:=0.0, Top:=0.0, Right:=0.0, Bottom:=0.0}
                        Widget:=texture_block{DefaultImage:=Textures.General.T_GenSquTrans, DefaultDesiredSize:=vector2{X:=850.0, Y:=130.0}, DefaultTint:=MakeColorFromHex("#0000FFFF")}

                    overlay_slot:
                        #Rarity level content
                        VerticalAlignment:=vertical_alignment.Center
                        HorizontalAlignment:=horizontal_alignment.Left
                        Padding:=margin{Left:=0.0, Top:=0.0, Right:=0.0, Bottom:=0.0}
                        Widget:=overlay:
                            Slots:=array:

                                overlay_slot:
                                    #Makeshift size box
                                    VerticalAlignment:=vertical_alignment.Center
                                    HorizontalAlignment:=horizontal_alignment.Center
                                    Padding:=margin{Left:=0.0, Top:=0.0, Right:=0.0, Bottom:=0.0}
                                    Widget:=texture_block{DefaultImage:=Textures.General.T_GenericSquare, DefaultDesiredSize:=vector2{X:=850.0, Y:=130.0}, DefaultTint:=MakeColorFromHex("#010a1832")}

                                overlay_slot:
                                    #Highlight texture, over the content but under the rarity flag
                                    VerticalAlignment:=vertical_alignment.Fill
                                    HorizontalAlignment:=horizontal_alignment.Fill
                                    Padding:=margin{Left:=0.0, Top:=0.0, Right:=0.0, Bottom:=0.0}
                                    Widget:=texture_block{DefaultImage:=Textures.General.T_LevelBorder, DefaultDesiredSize:=vector2{X:=512.0, Y:=64.0}, DefaultTint:=MakeColorFromHex("#1d4b89FF")}

                                    #Horizontal stack box
                                overlay_slot:
                                    VerticalAlignment:=vertical_alignment.Fill
                                    HorizontalAlignment:=horizontal_alignment.Left
                                    Padding:=margin{Left:=0.0, Top:=0.0, Right:=0.0, Bottom:=0.0}
                                    Widget:=stack_box:
                                        Orientation:=orientation.Horizontal
                                        Slots:=array:
                                            #Level flag
                                            stack_box_slot:
                                                VerticalAlignment:=vertical_alignment.Top
                                                HorizontalAlignment:=horizontal_alignment.Left
                                                Padding:=margin{Left:=0.0, Top:=0.0, Right:=15.0, Bottom:=0.0}
                                                Widget:=overlay:

                                                    Slots:=array:
                                                        overlay_slot:
                                                            VerticalAlignment:=vertical_alignment.Center
                                                            HorizontalAlignment:=horizontal_alignment.Center
                                                            Padding:=margin{Left:=0.0, Top:=0.0, Right:=0.0, Bottom:=0.0}
                                                            #Tier level makeshift size box
                                                            Widget:=texture_block{DefaultImage:=Textures.General.T_GenSquTrans, DefaultDesiredSize:=vector2{X:=150.0, Y:=130.0}, DefaultTint:=MakeColorFromHex("#FF0000FF")}

                                                        overlay_slot:
                                                            VerticalAlignment:=vertical_alignment.Top
                                                            HorizontalAlignment:=horizontal_alignment.Left
                                                            Padding:=margin{Left:=0.0, Top:=0.0, Right:=0.0, Bottom:=0.0}
                                                            #Selected level yellow flag
                                                            Widget:=stack_box:
                                                                Orientation:=orientation.Vertical
                                                                Slots:=array:
                                                                    stack_box_slot:
                                                                        VerticalAlignment:=vertical_alignment.Center
                                                                        HorizontalAlignment:=horizontal_alignment.Center
                                                                        Padding:=margin{Left:=0.0, Top:=0.0, Right:=0.0, Bottom:=0.0}
                                                                        #level flag texture
                                                                        Widget:=overlay:
                                                                            Slots:=array:

                                                                                overlay_slot:
                                                                                    VerticalAlignment:=vertical_alignment.Center
                                                                                    HorizontalAlignment:=horizontal_alignment.Center
                                                                                    Padding:=margin{Left:=0.0, Top:=0.0, Right:=0.0, Bottom:=0.0}
                                                                                    #level flag texture
                                                                                    Widget:=texture_block{DefaultImage:=Textures.General.T_GenLevelFlag, DefaultDesiredSize:=vector2{X:=150.0, Y:=45.0}, DefaultTint:=MakeColorFromHex("#f7ff1aFF")}

                                                                                overlay_slot:
                                                                                    VerticalAlignment:=vertical_alignment.Center
                                                                                    HorizontalAlignment:=horizontal_alignment.Center
                                                                                    Padding:=margin{Left:=0.0, Top:=0.0, Right:=0.0, Bottom:=0.0}
                                                                                    #level flag text
                                                                                    Widget:=text_block{DefaultText := StringToMessage("Base LV"), DefaultTextColor:=MakeColorFromHex("#000000FF")}

                                                                    stack_box_slot:
                                                                        VerticalAlignment:=vertical_alignment.Center
                                                                        HorizontalAlignment:=horizontal_alignment.Center
                                                                        Padding:=margin{Left:=0.0, Top:=0.0, Right:=0.0, Bottom:=0.0}
                                                                        #level flag texture
                                                                        Widget:=overlay:
                                                                            Slots:=array:

                                                                                #Booster Icon - Increase Max Shield
                                                                                overlay_slot:
                                                                                    VerticalAlignment:=vertical_alignment.Center
                                                                                    HorizontalAlignment:=horizontal_alignment.Center
                                                                                    Padding:=margin{Left:=0.0, Top:=0.0, Right:=0.0, Bottom:=0.0}
                                                                                    #Weapon icon BG
                                                                                    Widget:=texture_block{DefaultImage:=Textures.Rarity.T_IconBG_Common, DefaultDesiredSize:=vector2{X:=78.0, Y:=78.0}, DefaultTint:=MakeColorFromHex("#FFFFFFFF")}
                        
                                                                                overlay_slot:
                                                                                    VerticalAlignment:=vertical_alignment.Center
                                                                                    HorizontalAlignment:=horizontal_alignment.Center
                                                                                    Padding:=margin{Left:=0.0, Top:=0.0, Right:=0.0, Bottom:=0.0}
                                                                                    #Weapon icon Border
                                                                                    Widget:=texture_block{DefaultImage:=Textures.General.T_BoosterFrame, DefaultDesiredSize:=vector2{X:=78.0, Y:=78.0}, DefaultTint:=MakeColorFromHex("#FFFFFFFF")}
                        
                                                                                overlay_slot:
                                                                                    VerticalAlignment:=vertical_alignment.Center
                                                                                    HorizontalAlignment:=horizontal_alignment.Center
                                                                                    Padding:=margin{Left:=0.0, Top:=0.0, Right:=0.0, Bottom:=0.0}
                                                                                    #Weapon icon - pistol
                                                                                    Widget:=texture_block{DefaultImage:=Textures.Boosters.T_IncreaseMaxShield, DefaultDesiredSize:=vector2{X:=70.0, Y:=70.0}, DefaultTint:=MakeColorFromHex("#FFFFFFFF")}
                        
                                                                                #overlay_slot:
                                                                                    #VerticalAlignment:=vertical_alignment.Bottom
                                                                                    #HorizontalAlignment:=horizontal_alignment.Center
                                                                                    #Padding:=margin{Left:=0.0, Top:=0.0, Right:=0.0, Bottom:=5.0}
                                                                                    ##Weapon header - pistol
                                                                                    #Widget:=texture_block{DefaultImage:=Textures.Items.T_HeaderPistol, DefaultDesiredSize:=vector2{X:=78.0, Y:=18.0}, DefaultTint:=MakeColorFromHex("#FFFFFFFF")}

                                                        #overlay_slot:
                                                            #VerticalAlignment:=vertical_alignment.Fill
                                                            #HorizontalAlignment:=horizontal_alignment.Center
                                                            #Padding:=margin{Left:=0.0, Top:=5.0, Right:=0.0, Bottom:=5.0}
                                                            #Unselected level and lock
                                                            #Widget:=overlay:
                                                                #Slots:=array:
                                                                    
                                                                    #overlay_slot:
                                                                        #VerticalAlignment:=vertical_alignment.Top
                                                                        #HorizontalAlignment:=horizontal_alignment.Center
                                                                        #Padding:=margin{Left:=0.0, Top:=0.0, Right:=0.0, Bottom:=0.0}
                                                                        #level flag texture
                                                                        #Widget:=text_block{DefaultText := StringToMessage("Base Level"), DefaultTextColor:=MakeColorFromHex("#FFFFFFFF")}

                                                                    #overlay_slot:
                                                                        #VerticalAlignment:=vertical_alignment.Bottom
                                                                        #HorizontalAlignment:=horizontal_alignment.Center
                                                                        #Padding:=margin{Left:=0.0, Top:=0.0, Right:=0.0, Bottom:=0.0}
                                                                        ##level flag texture
                                                                        #Widget:=overlay:
                                                                            #Slots:=array:
                                                                                ##Weapon Icon - Pistols
                                                                                #overlay_slot:
                                                                                    #VerticalAlignment:=vertical_alignment.Center
                                                                                    #HorizontalAlignment:=horizontal_alignment.Center
                                                                                    #Padding:=margin{Left:=0.0, Top:=0.0, Right:=0.0, Bottom:=0.0}
                                                                                    ##Weapon icon BG
                                                                                    #Widget:=texture_block{DefaultImage:=Textures.General.T_GenericSquare, DefaultDesiredSize:=vector2{X:=78.0, Y:=78.0}, DefaultTint:=MakeColorFromHex("#000c1d88")}
                        
                                                                                #overlay_slot:
                                                                                    #VerticalAlignment:=vertical_alignment.Center
                                                                                    #HorizontalAlignment:=horizontal_alignment.Center
                                                                                    #Padding:=margin{Left:=0.0, Top:=0.0, Right:=0.0, Bottom:=0.0}
                                                                                    ##Weapon icon Border
                                                                                    #Widget:=texture_block{DefaultImage:=Textures.General.T_GenericBorderSquare, DefaultDesiredSize:=vector2{X:=78.0, Y:=78.0}, DefaultTint:=MakeColorFromHex("#2e7be3FF")}
                        
                                                                                #overlay_slot:
                                                                                    #VerticalAlignment:=vertical_alignment.Top
                                                                                    #HorizontalAlignment:=horizontal_alignment.Center
                                                                                    #Padding:=margin{Left:=0.0, Top:=5.0, Right:=0.0, Bottom:=0.0}
                                                                                    ##Weapon icon - pistol
                                                                                    #Widget:=texture_block{DefaultImage:=Textures.Items.T_IconPistol, DefaultDesiredSize:=vector2{X:=50.0, Y:=50.0}, DefaultTint:=MakeColorFromHex("#FFFFFFFF")}
                        
                                                                                #overlay_slot:
                                                                                    #VerticalAlignment:=vertical_alignment.Bottom
                                                                                    #HorizontalAlignment:=horizontal_alignment.Center
                                                                                    #Padding:=margin{Left:=0.0, Top:=0.0, Right:=0.0, Bottom:=5.0}
                                                                                    ##Weapon header - pistol
                                                                                    #Widget:=texture_block{DefaultImage:=Textures.Items.T_HeaderPistol, DefaultDesiredSize:=vector2{X:=78.0, Y:=18.0}, DefaultTint:=MakeColorFromHex("#FFFFFFFF")}
                                            
                                            

                                            stack_box_slot:
                                                VerticalAlignment:=vertical_alignment.Fill
                                                HorizontalAlignment:=horizontal_alignment.Center #Center
                                                Padding:=margin{Left:=0.0, Top:=0.0, Right:=0.0, Bottom:=0.0}
                                                Widget:=overlay:
                                                    Slots:=array:
                                                        overlay_slot:
                                                            #Makeshift size box
                                                            VerticalAlignment:=vertical_alignment.Center
                                                            HorizontalAlignment:=horizontal_alignment.Center
                                                            Padding:=margin{Left:=0.0, Top:=0.0, Right:=0.0, Bottom:=0.0}
                                                            Widget:=texture_block{DefaultImage:=Textures.General.T_GenSquTrans, DefaultDesiredSize:=vector2{X:=418.0, Y:=100.0}, DefaultTint:=MakeColorFromHex("#00FF0011")}

                                                        overlay_slot:
                                                            #Vertical stack box to add all the rarity information
                                                            VerticalAlignment:=vertical_alignment.Center
                                                            HorizontalAlignment:=horizontal_alignment.Center
                                                            Padding:=margin{Left:=0.0, Top:=0.0, Right:=0.0, Bottom:=0.0}
                                                            Widget:=stack_box:
                                                                Orientation:=orientation.Vertical
                                                                Slots:=array:
                                                                    stack_box_slot:
                                                                        #Horizontal Stack box for max rarity available
                                                                        VerticalAlignment:=vertical_alignment.Top
                                                                        HorizontalAlignment:=horizontal_alignment.Center
                                                                        Padding:=margin{Left:=0.0, Top:=0.0, Right:=0.0, Bottom:=0.0}
                                                                        Widget:=stack_box:
                                                                            Orientation:=orientation.Horizontal
                                                                            Slots:=array:
                                                                                stack_box_slot:
                                                                                    VerticalAlignment:=vertical_alignment.Center
                                                                                    HorizontalAlignment:=horizontal_alignment.Center
                                                                                    Padding:=margin{Left:=0.0, Top:=0.0, Right:=0.0, Bottom:=0.0}
                                                                                    Widget:=text_block{DefaultText := StringToMessage("Increase Max Shield"), DefaultTextColor:=MakeColorFromHex("#FFFFFFFF")}

                                                                    stack_box_slot:
                                                                        #Horizontal overlay for overall rarity scheme
                                                                        VerticalAlignment:=vertical_alignment.Bottom
                                                                        HorizontalAlignment:=horizontal_alignment.Center
                                                                        Padding:=margin{Left:=0.0, Top:=0.0, Right:=0.0, Bottom:=0.0}
                                                                        Widget:=overlay:
                                                                            Slots:=array:
                                                                                
                                                                                #Makeshift size box for the rarity scheme
                                                                                overlay_slot:
                                                                                    VerticalAlignment:=vertical_alignment.Center
                                                                                    HorizontalAlignment:=horizontal_alignment.Center
                                                                                    Padding:=margin{Left:=0.0, Top:=0.0, Right:=0.0, Bottom:=0.0}
                                                                                    Widget:=texture_block{DefaultImage:=Textures.General.T_GenSquTrans, DefaultDesiredSize:=vector2{X:=255.0, Y:=50.0}, DefaultTint:=MakeColorFromHex("#FF000022")}#Red frame

                                                                                #Visible border for rarity layout
                                                                                overlay_slot:
                                                                                    VerticalAlignment:=vertical_alignment.Center
                                                                                    HorizontalAlignment:=horizontal_alignment.Center
                                                                                    Padding:=margin{Left:=0.0, Top:=0.0, Right:=0.0, Bottom:=0.0}
                                                                                    Widget:=texture_block{DefaultImage:=Textures.General.T_GenericBorderSquare3, DefaultDesiredSize:=vector2{X:=255.0, Y:=40.0}, DefaultTint:=MakeColorFromHex("#2e7be3FF")}

                                                                                #Horizontal Stack box for rarity slots
                                                                                overlay_slot:
                                                                                    VerticalAlignment:=vertical_alignment.Fill
                                                                                    HorizontalAlignment:=horizontal_alignment.Fill
                                                                                    Padding:=margin{Left:=5.0, Top:=0.0, Right:=5.0, Bottom:=0.0}
                                                                                    Widget:=stack_box:
                                                                                        Orientation:=orientation.Horizontal
                                                                                        Slots:=array:
                                                                                            #Overlay for each rarity slots, including the pointing arrow
                                                                                            stack_box_slot:
                                                                                                VerticalAlignment:=vertical_alignment.Center
                                                                                                HorizontalAlignment:=horizontal_alignment.Center
                                                                                                Distribution:=option{1.0}
                                                                                                Padding:=margin{Left:=0.0, Top:=0.0, Right:=0.0, Bottom:=0.0}
                                                                                                Widget:=overlay:
                                                                                                    Slots:=array:
                                                                                                        #Makeshift size box
                                                                                                        overlay_slot:
                                                                                                            VerticalAlignment:=vertical_alignment.Center
                                                                                                            HorizontalAlignment:=horizontal_alignment.Center
                                                                                                            Padding:=margin{Left:=0.0, Top:=0.0, Right:=0.0, Bottom:=0.0}
                                                                                                            Widget:=texture_block{DefaultImage:=Textures.General.T_GenSquTrans, DefaultDesiredSize:=vector2{X:=30.0, Y:=40.0}, DefaultTint:=MakeColorFromHex("#0000FF08")}

                                                                                                        #Overlay for rarity border and texture
                                                                                                        overlay_slot:
                                                                                                            VerticalAlignment:=vertical_alignment.Center
                                                                                                            HorizontalAlignment:=horizontal_alignment.Center
                                                                                                            Padding:=margin{Left:=0.0, Top:=0.0, Right:=0.0, Bottom:=0.0}
                                                                                                            Widget:=overlay:
                                                                                                                Slots:=array:
                                                                                                                    
                                                                                                                    #Rarity texture border, when the rarity slot is empty
                                                                                                                    #overlay_slot:
                                                                                                                        #VerticalAlignment:=vertical_alignment.Center
                                                                                                                        #HorizontalAlignment:=horizontal_alignment.Center
                                                                                                                        #Padding:=margin{Left:=0.0, Top:=0.0, Right:=0.0, Bottom:=0.0}
                                                                                                                        #Widget:=texture_block{DefaultImage:=Textures.General.T_GenericBorderSquare2, DefaultDesiredSize:=vector2{X:=30.0, Y:=30.0}, DefaultTint:=MakeColorFromHex("#2e7be388")}

                                                                                                                    #Rarity texture
                                                                                                                    overlay_slot:
                                                                                                                        VerticalAlignment:=vertical_alignment.Center
                                                                                                                        HorizontalAlignment:=horizontal_alignment.Center
                                                                                                                        Padding:=margin{Left:=0.0, Top:=0.0, Right:=0.0, Bottom:=0.0}
                                                                                                                        Widget:=texture_block{DefaultImage:=Textures.General.T_GenericSquare, DefaultDesiredSize:=vector2{X:=20.0, Y:=20.0}, DefaultTint:=MakeColorFromHex("#979797FF")}
                                                                                                        
                                                                                                                    

                                                                                                        #Bracket
                                                                                                        overlay_slot:
                                                                                                            VerticalAlignment:=vertical_alignment.Fill
                                                                                                            HorizontalAlignment:=horizontal_alignment.Fill
                                                                                                            Padding:=margin{Left:=0.0, Top:=0.0, Right:=0.0, Bottom:=0.0}
                                                                                                            Widget:=stack_box:
                                                                                                                Orientation:=orientation.Horizontal
                                                                                                                #two instances of bracket texture to represent the range of rarities
                                                                                                                Slots:=array:

                                                                                                                    stack_box_slot:
                                                                                                                        VerticalAlignment:=vertical_alignment.Center
                                                                                                                        HorizontalAlignment:=horizontal_alignment.Left
                                                                                                                        Distribution:=option{1.0}
                                                                                                                        Padding:=margin{Left:=0.0, Top:=0.0, Right:=0.0, Bottom:=0.0}
                                                                                                                        Widget:=texture_block{DefaultImage:=Textures.General.T_RarityBracketLeft, DefaultDesiredSize:=vector2{X:=18.0, Y:=36.0}, DefaultTint:=MakeColorFromHex("#FFFFFF44")}

                                                                                                                    stack_box_slot:
                                                                                                                        VerticalAlignment:=vertical_alignment.Center
                                                                                                                        HorizontalAlignment:=horizontal_alignment.Right
                                                                                                                        Distribution:=option{1.0}
                                                                                                                        Padding:=margin{Left:=0.0, Top:=0.0, Right:=0.0, Bottom:=0.0}
                                                                                                                        Widget:=texture_block{DefaultImage:=Textures.General.T_RarityBracketRight, DefaultDesiredSize:=vector2{X:=18.0, Y:=36.0}, DefaultTint:=MakeColorFromHex("#FFFFFF44")}

                                                                                            stack_box_slot:
                                                                                                VerticalAlignment:=vertical_alignment.Center
                                                                                                HorizontalAlignment:=horizontal_alignment.Center
                                                                                                Distribution:=option{1.0}
                                                                                                Padding:=margin{Left:=0.0, Top:=0.0, Right:=0.0, Bottom:=0.0}
                                                                                                Widget:=overlay:
                                                                                                    Slots:=array:
                                                                                                        #Makeshift size box
                                                                                                        overlay_slot:
                                                                                                            VerticalAlignment:=vertical_alignment.Center
                                                                                                            HorizontalAlignment:=horizontal_alignment.Center
                                                                                                            Padding:=margin{Left:=0.0, Top:=0.0, Right:=0.0, Bottom:=0.0}
                                                                                                            Widget:=texture_block{DefaultImage:=Textures.General.T_GenSquTrans, DefaultDesiredSize:=vector2{X:=30.0, Y:=40.0}, DefaultTint:=MakeColorFromHex("#0000FF08")}

                                                                                                        #Overlay for rarity border and texture
                                                                                                        overlay_slot:
                                                                                                            VerticalAlignment:=vertical_alignment.Center
                                                                                                            HorizontalAlignment:=horizontal_alignment.Center
                                                                                                            Padding:=margin{Left:=0.0, Top:=0.0, Right:=0.0, Bottom:=0.0}
                                                                                                            Widget:=overlay:
                                                                                                                Slots:=array:
                                                                                                                    
                                                                                                                    #Rarity texture border, when the rarity slot is empty
                                                                                                                    overlay_slot:
                                                                                                                        VerticalAlignment:=vertical_alignment.Center
                                                                                                                        HorizontalAlignment:=horizontal_alignment.Center
                                                                                                                        Padding:=margin{Left:=0.0, Top:=0.0, Right:=0.0, Bottom:=0.0}
                                                                                                                        Widget:=texture_block{DefaultImage:=Textures.General.T_GenericBorderSquare2, DefaultDesiredSize:=vector2{X:=30.0, Y:=30.0}, DefaultTint:=MakeColorFromHex("#2e7be388")}

                                                                                                                    #Rarity texture
                                                                                                                    #overlay_slot:
                                                                                                                        #VerticalAlignment:=vertical_alignment.Center
                                                                                                                        #HorizontalAlignment:=horizontal_alignment.Center
                                                                                                                        #Padding:=margin{Left:=0.0, Top:=0.0, Right:=0.0, Bottom:=0.0}
                                                                                                                        #Widget:=texture_block{DefaultImage:=Textures.General.T_GenericSquare, DefaultDesiredSize:=vector2{X:=30.0, Y:=30.0}, DefaultTint:=MakeColorFromHex("#4ccc2aFF")}

                                                                                                        #Bracket
                                                                                                        #overlay_slot:
                                                                                                            #VerticalAlignment:=vertical_alignment.Fill
                                                                                                            #HorizontalAlignment:=horizontal_alignment.Fill
                                                                                                            #Padding:=margin{Left:=0.0, Top:=0.0, Right:=0.0, Bottom:=0.0}
                                                                                                            #Widget:=stack_box:
                                                                                                                #Orientation:=orientation.Horizontal
                                                                                                                #two instances of bracket texture to represent the range of rarities
                                                                                                                #Slots:=array:

                                                                                                                    #stack_box_slot:
                                                                                                                        #VerticalAlignment:=vertical_alignment.Center
                                                                                                                        #HorizontalAlignment:=horizontal_alignment.Left
                                                                                                                        #Distribution:=option{1.0}
                                                                                                                        #Padding:=margin{Left:=0.0, Top:=0.0, Right:=0.0, Bottom:=0.0}
                                                                                                                        #Widget:=texture_block{DefaultImage:=Textures.General.T_RarityBracketLeft, DefaultDesiredSize:=vector2{X:=18.0, Y:=36.0}, DefaultTint:=MakeColorFromHex("#FFFFFF44")}

                                                                                                                    #stack_box_slot:
                                                                                                                        #VerticalAlignment:=vertical_alignment.Center
                                                                                                                        #HorizontalAlignment:=horizontal_alignment.Right
                                                                                                                        #Distribution:=option{1.0}
                                                                                                                        #Padding:=margin{Left:=0.0, Top:=0.0, Right:=0.0, Bottom:=0.0}
                                                                                                                        #Widget:=texture_block{DefaultImage:=Textures.General.T_RarityBracketRight, DefaultDesiredSize:=vector2{X:=18.0, Y:=36.0}, DefaultTint:=MakeColorFromHex("#FFFFFF44")}

                                                                                            stack_box_slot:
                                                                                                VerticalAlignment:=vertical_alignment.Center
                                                                                                HorizontalAlignment:=horizontal_alignment.Center
                                                                                                Distribution:=option{1.0}
                                                                                                Padding:=margin{Left:=0.0, Top:=0.0, Right:=0.0, Bottom:=0.0}
                                                                                                Widget:=overlay:
                                                                                                    Slots:=array:
                                                                                                        #Makeshift size box
                                                                                                        overlay_slot:
                                                                                                            VerticalAlignment:=vertical_alignment.Center
                                                                                                            HorizontalAlignment:=horizontal_alignment.Center
                                                                                                            Padding:=margin{Left:=0.0, Top:=0.0, Right:=0.0, Bottom:=0.0}
                                                                                                            Widget:=texture_block{DefaultImage:=Textures.General.T_GenSquTrans, DefaultDesiredSize:=vector2{X:=30.0, Y:=40.0}, DefaultTint:=MakeColorFromHex("#0000FF08")}

                                                                                                        #Overlay for rarity border and texture
                                                                                                        overlay_slot:
                                                                                                            VerticalAlignment:=vertical_alignment.Center
                                                                                                            HorizontalAlignment:=horizontal_alignment.Center
                                                                                                            Padding:=margin{Left:=0.0, Top:=0.0, Right:=0.0, Bottom:=0.0}
                                                                                                            Widget:=overlay:
                                                                                                                Slots:=array:
                                                                                                                                                                                                                
                                                                                                                    #Rarity texture border, when the rarity slot is empty
                                                                                                                    overlay_slot:
                                                                                                                        VerticalAlignment:=vertical_alignment.Center
                                                                                                                        HorizontalAlignment:=horizontal_alignment.Center
                                                                                                                        Padding:=margin{Left:=0.0, Top:=0.0, Right:=0.0, Bottom:=0.0}
                                                                                                                        Widget:=texture_block{DefaultImage:=Textures.General.T_GenericBorderSquare2, DefaultDesiredSize:=vector2{X:=30.0, Y:=30.0}, DefaultTint:=MakeColorFromHex("#2e7be388")}
                                                                                                                                                                                                                            
                                                                                                                    #Rarity texture
                                                                                                                    #overlay_slot:
                                                                                                                        #VerticalAlignment:=vertical_alignment.Bottom
                                                                                                                        #HorizontalAlignment:=horizontal_alignment.Center
                                                                                                                        #Padding:=margin{Left:=0.0, Top:=0.0, Right:=0.0, Bottom:=0.0}
                                                                                                                        #Widget:=texture_block{DefaultImage:=Textures.General.T_GenericSquare, DefaultDesiredSize:=vector2{X:=30.0, Y:=30.0}, DefaultTint:=MakeColorFromHex("#2a54ccFF")}
                                                                                                                    
                                                                                                        #Bracket
                                                                                                        #overlay_slot:
                                                                                                            #VerticalAlignment:=vertical_alignment.Fill
                                                                                                            #HorizontalAlignment:=horizontal_alignment.Fill
                                                                                                            #Padding:=margin{Left:=0.0, Top:=0.0, Right:=0.0, Bottom:=0.0}
                                                                                                            #Widget:=stack_box:
                                                                                                                #Orientation:=orientation.Horizontal
                                                                                                                #two instances of bracket texture to represent the range of rarities
                                                                                                                #Slots:=array:

                                                                                                                    #stack_box_slot:
                                                                                                                        #VerticalAlignment:=vertical_alignment.Center
                                                                                                                        #HorizontalAlignment:=horizontal_alignment.Left
                                                                                                                        #Distribution:=option{1.0}
                                                                                                                        #Padding:=margin{Left:=0.0, Top:=0.0, Right:=0.0, Bottom:=0.0}
                                                                                                                        #Widget:=texture_block{DefaultImage:=Textures.General.T_RarityBracketLeft, DefaultDesiredSize:=vector2{X:=18.0, Y:=36.0}, DefaultTint:=MakeColorFromHex("#FFFFFF44")}

                                                                                                                    #stack_box_slot:
                                                                                                                        #VerticalAlignment:=vertical_alignment.Center
                                                                                                                        #HorizontalAlignment:=horizontal_alignment.Right
                                                                                                                        #Distribution:=option{1.0}
                                                                                                                        #Padding:=margin{Left:=0.0, Top:=0.0, Right:=0.0, Bottom:=0.0}
                                                                                                                        #Widget:=texture_block{DefaultImage:=Textures.General.T_RarityBracketRight, DefaultDesiredSize:=vector2{X:=18.0, Y:=36.0}, DefaultTint:=MakeColorFromHex("#FFFFFF44")}

                                                                                            stack_box_slot:
                                                                                                VerticalAlignment:=vertical_alignment.Center
                                                                                                HorizontalAlignment:=horizontal_alignment.Center
                                                                                                Distribution:=option{1.0}
                                                                                                Padding:=margin{Left:=0.0, Top:=0.0, Right:=0.0, Bottom:=0.0}
                                                                                                Widget:=overlay:
                                                                                                    Slots:=array:
                                                                                                        #Makeshift size box
                                                                                                        overlay_slot:
                                                                                                            VerticalAlignment:=vertical_alignment.Center
                                                                                                            HorizontalAlignment:=horizontal_alignment.Center
                                                                                                            Padding:=margin{Left:=0.0, Top:=0.0, Right:=0.0, Bottom:=0.0}
                                                                                                            Widget:=texture_block{DefaultImage:=Textures.General.T_GenSquTrans, DefaultDesiredSize:=vector2{X:=30.0, Y:=40.0}, DefaultTint:=MakeColorFromHex("#0000FF08")}

                                                                                                        #Overlay for rarity border and texture
                                                                                                        overlay_slot:
                                                                                                            VerticalAlignment:=vertical_alignment.Center
                                                                                                            HorizontalAlignment:=horizontal_alignment.Center
                                                                                                            Padding:=margin{Left:=0.0, Top:=0.0, Right:=0.0, Bottom:=0.0}
                                                                                                            Widget:=overlay:
                                                                                                                Slots:=array:

                                                                                                                    #Rarity texture border, when the rarity slot is empty
                                                                                                                    overlay_slot:
                                                                                                                        VerticalAlignment:=vertical_alignment.Center
                                                                                                                        HorizontalAlignment:=horizontal_alignment.Center
                                                                                                                        Padding:=margin{Left:=0.0, Top:=0.0, Right:=0.0, Bottom:=0.0}
                                                                                                                        Widget:=texture_block{DefaultImage:=Textures.General.T_GenericBorderSquare2, DefaultDesiredSize:=vector2{X:=30.0, Y:=30.0}, DefaultTint:=MakeColorFromHex("#2e7be388")}
                                                                                                                    
                                                                                                                    #Rarity texture
                                                                                                                    #overlay_slot:
                                                                                                                        #VerticalAlignment:=vertical_alignment.Bottom
                                                                                                                        #HorizontalAlignment:=horizontal_alignment.Center
                                                                                                                        #Padding:=margin{Left:=0.0, Top:=0.0, Right:=0.0, Bottom:=0.0}
                                                                                                                        #Widget:=texture_block{DefaultImage:=Textures.General.T_GenericSquare, DefaultDesiredSize:=vector2{X:=30.0, Y:=30.0}, DefaultTint:=MakeColorFromHex("#b32accFF")}
                                                                                                                    
                                                                                                        #Bracket
                                                                                                        #overlay_slot:
                                                                                                            #VerticalAlignment:=vertical_alignment.Fill
                                                                                                            #HorizontalAlignment:=horizontal_alignment.Fill
                                                                                                            #Padding:=margin{Left:=0.0, Top:=0.0, Right:=0.0, Bottom:=0.0}
                                                                                                            #Widget:=stack_box:
                                                                                                                #Orientation:=orientation.Horizontal
                                                                                                                #two instances of bracket texture to represent the range of rarities
                                                                                                                #Slots:=array:

                                                                                                                    #stack_box_slot:
                                                                                                                        #VerticalAlignment:=vertical_alignment.Center
                                                                                                                        #HorizontalAlignment:=horizontal_alignment.Left
                                                                                                                        #Distribution:=option{1.0}
                                                                                                                        #Padding:=margin{Left:=0.0, Top:=0.0, Right:=0.0, Bottom:=0.0}
                                                                                                                        #Widget:=texture_block{DefaultImage:=Textures.General.T_RarityBracketLeft, DefaultDesiredSize:=vector2{X:=18.0, Y:=36.0}, DefaultTint:=MakeColorFromHex("#FFFFFF44")}

                                                                                                                    #stack_box_slot:
                                                                                                                        #VerticalAlignment:=vertical_alignment.Center
                                                                                                                        #HorizontalAlignment:=horizontal_alignment.Right
                                                                                                                        #Distribution:=option{1.0}
                                                                                                                        #Padding:=margin{Left:=0.0, Top:=0.0, Right:=0.0, Bottom:=0.0}
                                                                                                                        #Widget:=texture_block{DefaultImage:=Textures.General.T_RarityBracketRight, DefaultDesiredSize:=vector2{X:=18.0, Y:=36.0}, DefaultTint:=MakeColorFromHex("#FFFFFF44")}

                                                                                            stack_box_slot:
                                                                                                VerticalAlignment:=vertical_alignment.Center
                                                                                                HorizontalAlignment:=horizontal_alignment.Center
                                                                                                Distribution:=option{1.0}
                                                                                                Padding:=margin{Left:=0.0, Top:=0.0, Right:=0.0, Bottom:=0.0}
                                                                                                Widget:=overlay:
                                                                                                    Slots:=array:
                                                                                                        #Makeshift size box
                                                                                                        overlay_slot:
                                                                                                            VerticalAlignment:=vertical_alignment.Center
                                                                                                            HorizontalAlignment:=horizontal_alignment.Center
                                                                                                            Padding:=margin{Left:=0.0, Top:=0.0, Right:=0.0, Bottom:=0.0}
                                                                                                            Widget:=texture_block{DefaultImage:=Textures.General.T_GenSquTrans, DefaultDesiredSize:=vector2{X:=30.0, Y:=40.0}, DefaultTint:=MakeColorFromHex("#0000FF08")}

                                                                                                        #Overlay for rarity border and texture
                                                                                                        overlay_slot:
                                                                                                            VerticalAlignment:=vertical_alignment.Center
                                                                                                            HorizontalAlignment:=horizontal_alignment.Center
                                                                                                            Padding:=margin{Left:=0.0, Top:=0.0, Right:=0.0, Bottom:=0.0}
                                                                                                            Widget:=overlay:
                                                                                                                Slots:=array:

                                                                                                                    #Rarity texture border, when the rarity slot is empty
                                                                                                                    overlay_slot:
                                                                                                                        VerticalAlignment:=vertical_alignment.Center
                                                                                                                        HorizontalAlignment:=horizontal_alignment.Center
                                                                                                                        Padding:=margin{Left:=0.0, Top:=0.0, Right:=0.0, Bottom:=0.0}
                                                                                                                        Widget:=texture_block{DefaultImage:=Textures.General.T_GenericBorderSquare2, DefaultDesiredSize:=vector2{X:=30.0, Y:=30.0}, DefaultTint:=MakeColorFromHex("#2e7be388")}
                                                                                                                    
                                                                                                                    #Rarity texture
                                                                                                                    #overlay_slot:
                                                                                                                        #VerticalAlignment:=vertical_alignment.Bottom
                                                                                                                        #HorizontalAlignment:=horizontal_alignment.Center
                                                                                                                        #Padding:=margin{Left:=0.0, Top:=0.0, Right:=0.0, Bottom:=0.0}
                                                                                                                        #Widget:=texture_block{DefaultImage:=Textures.General.T_GenericSquare, DefaultDesiredSize:=vector2{X:=30.0, Y:=30.0}, DefaultTint:=MakeColorFromHex("#cc8d2aFF")}

                                                                                                        #Bracket
                                                                                                        #overlay_slot:
                                                                                                            #VerticalAlignment:=vertical_alignment.Fill
                                                                                                            #HorizontalAlignment:=horizontal_alignment.Fill
                                                                                                            #Padding:=margin{Left:=0.0, Top:=0.0, Right:=0.0, Bottom:=0.0}
                                                                                                            #Widget:=stack_box:
                                                                                                                #Orientation:=orientation.Horizontal
                                                                                                                #two instances of bracket texture to represent the range of rarities
                                                                                                                #Slots:=array:

                                                                                                                    #stack_box_slot:
                                                                                                                        #VerticalAlignment:=vertical_alignment.Center
                                                                                                                        #HorizontalAlignment:=horizontal_alignment.Left
                                                                                                                        #Distribution:=option{1.0}
                                                                                                                        #Padding:=margin{Left:=0.0, Top:=0.0, Right:=0.0, Bottom:=0.0}
                                                                                                                        #Widget:=texture_block{DefaultImage:=Textures.General.T_RarityBracketLeft, DefaultDesiredSize:=vector2{X:=18.0, Y:=36.0}, DefaultTint:=MakeColorFromHex("#FFFFFF44")}

                                                                                                                    #stack_box_slot:
                                                                                                                        #VerticalAlignment:=vertical_alignment.Center
                                                                                                                        #HorizontalAlignment:=horizontal_alignment.Right
                                                                                                                        #Distribution:=option{1.0}
                                                                                                                        #Padding:=margin{Left:=0.0, Top:=0.0, Right:=0.0, Bottom:=0.0}
                                                                                                                        #Widget:=texture_block{DefaultImage:=Textures.General.T_RarityBracketRight, DefaultDesiredSize:=vector2{X:=18.0, Y:=36.0}, DefaultTint:=MakeColorFromHex("#FFFFFF44")}

                                                                                            stack_box_slot:
                                                                                                VerticalAlignment:=vertical_alignment.Center
                                                                                                HorizontalAlignment:=horizontal_alignment.Center
                                                                                                Distribution:=option{1.0}
                                                                                                Padding:=margin{Left:=0.0, Top:=0.0, Right:=0.0, Bottom:=0.0}
                                                                                                Widget:=overlay:
                                                                                                    Slots:=array:
                                                                                                        #Makeshift size box
                                                                                                        overlay_slot:
                                                                                                            VerticalAlignment:=vertical_alignment.Center
                                                                                                            HorizontalAlignment:=horizontal_alignment.Center
                                                                                                            Padding:=margin{Left:=0.0, Top:=0.0, Right:=0.0, Bottom:=0.0}
                                                                                                            Widget:=texture_block{DefaultImage:=Textures.General.T_GenSquTrans, DefaultDesiredSize:=vector2{X:=30.0, Y:=40.0}, DefaultTint:=MakeColorFromHex("#0000FF08")}

                                                                                                        #Overlay for rarity border and texture
                                                                                                        overlay_slot:
                                                                                                            VerticalAlignment:=vertical_alignment.Center
                                                                                                            HorizontalAlignment:=horizontal_alignment.Center
                                                                                                            Padding:=margin{Left:=0.0, Top:=0.0, Right:=0.0, Bottom:=0.0}
                                                                                                            Widget:=overlay:
                                                                                                                Slots:=array:
                                                                                                        
                                                                                                                    #Rarity texture border, when the rarity slot is empty
                                                                                                                    overlay_slot:
                                                                                                                        VerticalAlignment:=vertical_alignment.Center
                                                                                                                        HorizontalAlignment:=horizontal_alignment.Center
                                                                                                                        Padding:=margin{Left:=0.0, Top:=0.0, Right:=0.0, Bottom:=0.0}
                                                                                                                        Widget:=texture_block{DefaultImage:=Textures.General.T_GenericBorderSquare2, DefaultDesiredSize:=vector2{X:=30.0, Y:=30.0}, DefaultTint:=MakeColorFromHex("#2e7be388")}
                                                                                                                    
                                                                                                                    #Rarity texture
                                                                                                                    #overlay_slot:
                                                                                                                        #VerticalAlignment:=vertical_alignment.Bottom
                                                                                                                        #HorizontalAlignment:=horizontal_alignment.Center
                                                                                                                        #Padding:=margin{Left:=0.0, Top:=0.0, Right:=0.0, Bottom:=0.0}
                                                                                                                        #Widget:=texture_block{DefaultImage:=Textures.General.T_GenericSquare, DefaultDesiredSize:=vector2{X:=30.0, Y:=30.0}, DefaultTint:=MakeColorFromHex("#eacf19FF")}
                                                                                                                    
                                                                                                        #Bracket
                                                                                                        #overlay_slot:
                                                                                                            #VerticalAlignment:=vertical_alignment.Fill
                                                                                                            #HorizontalAlignment:=horizontal_alignment.Fill
                                                                                                            #Padding:=margin{Left:=0.0, Top:=0.0, Right:=0.0, Bottom:=0.0}
                                                                                                            #Widget:=stack_box:
                                                                                                                #Orientation:=orientation.Horizontal
                                                                                                                #two instances of bracket texture to represent the range of rarities
                                                                                                                #Slots:=array:

                                                                                                                    #stack_box_slot:
                                                                                                                        #VerticalAlignment:=vertical_alignment.Center
                                                                                                                        #HorizontalAlignment:=horizontal_alignment.Left
                                                                                                                        #Distribution:=option{1.0}
                                                                                                                        #Padding:=margin{Left:=0.0, Top:=0.0, Right:=0.0, Bottom:=0.0}
                                                                                                                        #Widget:=texture_block{DefaultImage:=Textures.General.T_RarityBracketLeft, DefaultDesiredSize:=vector2{X:=18.0, Y:=36.0}, DefaultTint:=MakeColorFromHex("#FFFFFF44")}

                                                                                                                    #stack_box_slot:
                                                                                                                        #VerticalAlignment:=vertical_alignment.Center
                                                                                                                        #HorizontalAlignment:=horizontal_alignment.Right
                                                                                                                        #Distribution:=option{1.0}
                                                                                                                        #Padding:=margin{Left:=0.0, Top:=0.0, Right:=0.0, Bottom:=0.0}
                                                                                                                        #Widget:=texture_block{DefaultImage:=Textures.General.T_RarityBracketRight, DefaultDesiredSize:=vector2{X:=18.0, Y:=36.0}, DefaultTint:=MakeColorFromHex("#FFFFFF44")}

                                            #Upgrade button
                                            stack_box_slot:
                                                VerticalAlignment:=vertical_alignment.Center
                                                HorizontalAlignment:=horizontal_alignment.Center
                                                Padding:=margin{Left:=0.0, Top:=0.0, Right:=0.0, Bottom:=0.0}
                                                Widget:=button_loud{DefaultText:=StringToMessage("UPDATE\n$15")}

                    #overlay_slot:
                        #VerticalAlignment:=vertical_alignment.Fill
                        #HorizontalAlignment:=horizontal_alignment.Fill
                        #Padding:=margin{Left:=0.0, Top:=0.0, Right:=0.0, Bottom:=0.0}
                        #Widget:=texture_block{DefaultImage:=Textures.General.T_LevelPointer, DefaultDesiredSize:=vector2{X:=20.0, Y:=20.0}, DefaultTint:=MakeColorFromHex("#7de2FFFF")}

            LevelRow2:=overlay:
                Slots:=array:
                    overlay_slot:
                        #Makeshift size box
                        VerticalAlignment:=vertical_alignment.Center
                        HorizontalAlignment:=horizontal_alignment.Center
                        Padding:=margin{Left:=0.0, Top:=0.0, Right:=0.0, Bottom:=0.0}
                        Widget:=texture_block{DefaultImage:=Textures.General.T_GenSquTrans, DefaultDesiredSize:=vector2{X:=850.0, Y:=130.0}, DefaultTint:=MakeColorFromHex("#0000FFFF")}

                    overlay_slot:
                        #Rarity level content
                        VerticalAlignment:=vertical_alignment.Center
                        HorizontalAlignment:=horizontal_alignment.Left
                        Padding:=margin{Left:=0.0, Top:=0.0, Right:=0.0, Bottom:=0.0}
                        Widget:=overlay:
                            Slots:=array:

                                overlay_slot:
                                    #Makeshift size box
                                    VerticalAlignment:=vertical_alignment.Center
                                    HorizontalAlignment:=horizontal_alignment.Center
                                    Padding:=margin{Left:=0.0, Top:=0.0, Right:=0.0, Bottom:=0.0}
                                    Widget:=texture_block{DefaultImage:=Textures.General.T_GenericSquare, DefaultDesiredSize:=vector2{X:=850.0, Y:=130.0}, DefaultTint:=MakeColorFromHex("#010a1832")}

                                overlay_slot:
                                    #Highlight texture, over the content but under the rarity flag
                                    VerticalAlignment:=vertical_alignment.Fill
                                    HorizontalAlignment:=horizontal_alignment.Fill
                                    Padding:=margin{Left:=0.0, Top:=0.0, Right:=0.0, Bottom:=0.0}
                                    Widget:=texture_block{DefaultImage:=Textures.General.T_LevelBorder, DefaultDesiredSize:=vector2{X:=512.0, Y:=64.0}, DefaultTint:=MakeColorFromHex("#1d4b89FF")}

                                    #Horizontal stack box
                                overlay_slot:
                                    VerticalAlignment:=vertical_alignment.Fill
                                    HorizontalAlignment:=horizontal_alignment.Left
                                    Padding:=margin{Left:=0.0, Top:=0.0, Right:=0.0, Bottom:=0.0}
                                    Widget:=stack_box:
                                        Orientation:=orientation.Horizontal
                                        Slots:=array:
                                            #Level flag
                                            stack_box_slot:
                                                VerticalAlignment:=vertical_alignment.Top
                                                HorizontalAlignment:=horizontal_alignment.Left
                                                Padding:=margin{Left:=0.0, Top:=0.0, Right:=15.0, Bottom:=0.0}
                                                Widget:=overlay:

                                                    Slots:=array:
                                                        overlay_slot:
                                                            VerticalAlignment:=vertical_alignment.Center
                                                            HorizontalAlignment:=horizontal_alignment.Center
                                                            Padding:=margin{Left:=0.0, Top:=0.0, Right:=0.0, Bottom:=0.0}
                                                            #Tier level makeshift size box
                                                            Widget:=texture_block{DefaultImage:=Textures.General.T_GenSquTrans, DefaultDesiredSize:=vector2{X:=150.0, Y:=130.0}, DefaultTint:=MakeColorFromHex("#FF0000FF")}

                                                        overlay_slot:
                                                            VerticalAlignment:=vertical_alignment.Top
                                                            HorizontalAlignment:=horizontal_alignment.Left
                                                            Padding:=margin{Left:=0.0, Top:=0.0, Right:=0.0, Bottom:=0.0}
                                                            #Selected level yellow flag
                                                            Widget:=stack_box:
                                                                Orientation:=orientation.Vertical
                                                                Slots:=array:
                                                                    stack_box_slot:
                                                                        VerticalAlignment:=vertical_alignment.Center
                                                                        HorizontalAlignment:=horizontal_alignment.Center
                                                                        Padding:=margin{Left:=0.0, Top:=0.0, Right:=0.0, Bottom:=0.0}
                                                                        #level flag texture
                                                                        Widget:=overlay:
                                                                            Slots:=array:

                                                                                overlay_slot:
                                                                                    VerticalAlignment:=vertical_alignment.Center
                                                                                    HorizontalAlignment:=horizontal_alignment.Center
                                                                                    Padding:=margin{Left:=0.0, Top:=0.0, Right:=0.0, Bottom:=0.0}
                                                                                    #level flag texture
                                                                                    Widget:=texture_block{DefaultImage:=Textures.General.T_GenLevelFlag, DefaultDesiredSize:=vector2{X:=150.0, Y:=45.0}, DefaultTint:=MakeColorFromHex("#f7ff1aFF")}

                                                                                overlay_slot:
                                                                                    VerticalAlignment:=vertical_alignment.Center
                                                                                    HorizontalAlignment:=horizontal_alignment.Center
                                                                                    Padding:=margin{Left:=0.0, Top:=0.0, Right:=0.0, Bottom:=0.0}
                                                                                    #level flag text
                                                                                    Widget:=text_block{DefaultText := StringToMessage("Base LV"), DefaultTextColor:=MakeColorFromHex("#000000FF")}

                                                                    stack_box_slot:
                                                                        VerticalAlignment:=vertical_alignment.Center
                                                                        HorizontalAlignment:=horizontal_alignment.Center
                                                                        Padding:=margin{Left:=0.0, Top:=0.0, Right:=0.0, Bottom:=0.0}
                                                                        #level flag texture
                                                                        Widget:=overlay:
                                                                            Slots:=array:

                                                                                #Booster Icon - Increase Max Shield
                                                                                overlay_slot:
                                                                                    VerticalAlignment:=vertical_alignment.Center
                                                                                    HorizontalAlignment:=horizontal_alignment.Center
                                                                                    Padding:=margin{Left:=0.0, Top:=0.0, Right:=0.0, Bottom:=0.0}
                                                                                    #Weapon icon BG
                                                                                    Widget:=texture_block{DefaultImage:=Textures.Rarity.T_IconBG_Common, DefaultDesiredSize:=vector2{X:=78.0, Y:=78.0}, DefaultTint:=MakeColorFromHex("#FFFFFFFF")}
                        
                                                                                overlay_slot:
                                                                                    VerticalAlignment:=vertical_alignment.Center
                                                                                    HorizontalAlignment:=horizontal_alignment.Center
                                                                                    Padding:=margin{Left:=0.0, Top:=0.0, Right:=0.0, Bottom:=0.0}
                                                                                    #Weapon icon Border
                                                                                    Widget:=texture_block{DefaultImage:=Textures.General.T_BoosterFrame, DefaultDesiredSize:=vector2{X:=78.0, Y:=78.0}, DefaultTint:=MakeColorFromHex("#FFFFFFFF")}
                        
                                                                                overlay_slot:
                                                                                    VerticalAlignment:=vertical_alignment.Center
                                                                                    HorizontalAlignment:=horizontal_alignment.Center
                                                                                    Padding:=margin{Left:=0.0, Top:=0.0, Right:=0.0, Bottom:=0.0}
                                                                                    #Weapon icon - pistol
                                                                                    Widget:=texture_block{DefaultImage:=Textures.Boosters.T_IncreaseMaxShield, DefaultDesiredSize:=vector2{X:=70.0, Y:=70.0}, DefaultTint:=MakeColorFromHex("#FFFFFFFF")}
                        
                                                                                #overlay_slot:
                                                                                    #VerticalAlignment:=vertical_alignment.Bottom
                                                                                    #HorizontalAlignment:=horizontal_alignment.Center
                                                                                    #Padding:=margin{Left:=0.0, Top:=0.0, Right:=0.0, Bottom:=5.0}
                                                                                    ##Weapon header - pistol
                                                                                    #Widget:=texture_block{DefaultImage:=Textures.Items.T_HeaderPistol, DefaultDesiredSize:=vector2{X:=78.0, Y:=18.0}, DefaultTint:=MakeColorFromHex("#FFFFFFFF")}

                                                        #overlay_slot:
                                                            #VerticalAlignment:=vertical_alignment.Fill
                                                            #HorizontalAlignment:=horizontal_alignment.Center
                                                            #Padding:=margin{Left:=0.0, Top:=5.0, Right:=0.0, Bottom:=5.0}
                                                            #Unselected level and lock
                                                            #Widget:=overlay:
                                                                #Slots:=array:
                                                                    
                                                                    #overlay_slot:
                                                                        #VerticalAlignment:=vertical_alignment.Top
                                                                        #HorizontalAlignment:=horizontal_alignment.Center
                                                                        #Padding:=margin{Left:=0.0, Top:=0.0, Right:=0.0, Bottom:=0.0}
                                                                        #level flag texture
                                                                        #Widget:=text_block{DefaultText := StringToMessage("Base Level"), DefaultTextColor:=MakeColorFromHex("#FFFFFFFF")}

                                                                    #overlay_slot:
                                                                        #VerticalAlignment:=vertical_alignment.Bottom
                                                                        #HorizontalAlignment:=horizontal_alignment.Center
                                                                        #Padding:=margin{Left:=0.0, Top:=0.0, Right:=0.0, Bottom:=0.0}
                                                                        ##level flag texture
                                                                        #Widget:=overlay:
                                                                            #Slots:=array:
                                                                                ##Weapon Icon - Pistols
                                                                                #overlay_slot:
                                                                                    #VerticalAlignment:=vertical_alignment.Center
                                                                                    #HorizontalAlignment:=horizontal_alignment.Center
                                                                                    #Padding:=margin{Left:=0.0, Top:=0.0, Right:=0.0, Bottom:=0.0}
                                                                                    ##Weapon icon BG
                                                                                    #Widget:=texture_block{DefaultImage:=Textures.General.T_GenericSquare, DefaultDesiredSize:=vector2{X:=78.0, Y:=78.0}, DefaultTint:=MakeColorFromHex("#000c1d88")}
                        
                                                                                #overlay_slot:
                                                                                    #VerticalAlignment:=vertical_alignment.Center
                                                                                    #HorizontalAlignment:=horizontal_alignment.Center
                                                                                    #Padding:=margin{Left:=0.0, Top:=0.0, Right:=0.0, Bottom:=0.0}
                                                                                    ##Weapon icon Border
                                                                                    #Widget:=texture_block{DefaultImage:=Textures.General.T_GenericBorderSquare, DefaultDesiredSize:=vector2{X:=78.0, Y:=78.0}, DefaultTint:=MakeColorFromHex("#2e7be3FF")}
                        
                                                                                #overlay_slot:
                                                                                    #VerticalAlignment:=vertical_alignment.Top
                                                                                    #HorizontalAlignment:=horizontal_alignment.Center
                                                                                    #Padding:=margin{Left:=0.0, Top:=5.0, Right:=0.0, Bottom:=0.0}
                                                                                    ##Weapon icon - pistol
                                                                                    #Widget:=texture_block{DefaultImage:=Textures.Items.T_IconPistol, DefaultDesiredSize:=vector2{X:=50.0, Y:=50.0}, DefaultTint:=MakeColorFromHex("#FFFFFFFF")}
                        
                                                                                #overlay_slot:
                                                                                    #VerticalAlignment:=vertical_alignment.Bottom
                                                                                    #HorizontalAlignment:=horizontal_alignment.Center
                                                                                    #Padding:=margin{Left:=0.0, Top:=0.0, Right:=0.0, Bottom:=5.0}
                                                                                    ##Weapon header - pistol
                                                                                    #Widget:=texture_block{DefaultImage:=Textures.Items.T_HeaderPistol, DefaultDesiredSize:=vector2{X:=78.0, Y:=18.0}, DefaultTint:=MakeColorFromHex("#FFFFFFFF")}
                                            
                                            

                                            stack_box_slot:
                                                VerticalAlignment:=vertical_alignment.Fill
                                                HorizontalAlignment:=horizontal_alignment.Center #Center
                                                Padding:=margin{Left:=0.0, Top:=0.0, Right:=0.0, Bottom:=0.0}
                                                Widget:=overlay:
                                                    Slots:=array:
                                                        overlay_slot:
                                                            #Makeshift size box
                                                            VerticalAlignment:=vertical_alignment.Center
                                                            HorizontalAlignment:=horizontal_alignment.Center
                                                            Padding:=margin{Left:=0.0, Top:=0.0, Right:=0.0, Bottom:=0.0}
                                                            Widget:=texture_block{DefaultImage:=Textures.General.T_GenSquTrans, DefaultDesiredSize:=vector2{X:=418.0, Y:=100.0}, DefaultTint:=MakeColorFromHex("#00FF0011")}

                                                        overlay_slot:
                                                            #Vertical stack box to add all the rarity information
                                                            VerticalAlignment:=vertical_alignment.Center
                                                            HorizontalAlignment:=horizontal_alignment.Center
                                                            Padding:=margin{Left:=0.0, Top:=0.0, Right:=0.0, Bottom:=0.0}
                                                            Widget:=stack_box:
                                                                Orientation:=orientation.Vertical
                                                                Slots:=array:
                                                                    stack_box_slot:
                                                                        #Horizontal Stack box for max rarity available
                                                                        VerticalAlignment:=vertical_alignment.Top
                                                                        HorizontalAlignment:=horizontal_alignment.Center
                                                                        Padding:=margin{Left:=0.0, Top:=0.0, Right:=0.0, Bottom:=0.0}
                                                                        Widget:=stack_box:
                                                                            Orientation:=orientation.Horizontal
                                                                            Slots:=array:
                                                                                stack_box_slot:
                                                                                    VerticalAlignment:=vertical_alignment.Center
                                                                                    HorizontalAlignment:=horizontal_alignment.Center
                                                                                    Padding:=margin{Left:=0.0, Top:=0.0, Right:=0.0, Bottom:=0.0}
                                                                                    Widget:=text_block{DefaultText := StringToMessage("Increase Max Shield"), DefaultTextColor:=MakeColorFromHex("#FFFFFFFF")}

                                                                    stack_box_slot:
                                                                        #Horizontal overlay for overall rarity scheme
                                                                        VerticalAlignment:=vertical_alignment.Bottom
                                                                        HorizontalAlignment:=horizontal_alignment.Center
                                                                        Padding:=margin{Left:=0.0, Top:=0.0, Right:=0.0, Bottom:=0.0}
                                                                        Widget:=overlay:
                                                                            Slots:=array:
                                                                                
                                                                                #Makeshift size box for the rarity scheme
                                                                                overlay_slot:
                                                                                    VerticalAlignment:=vertical_alignment.Center
                                                                                    HorizontalAlignment:=horizontal_alignment.Center
                                                                                    Padding:=margin{Left:=0.0, Top:=0.0, Right:=0.0, Bottom:=0.0}
                                                                                    Widget:=texture_block{DefaultImage:=Textures.General.T_GenSquTrans, DefaultDesiredSize:=vector2{X:=255.0, Y:=50.0}, DefaultTint:=MakeColorFromHex("#FF000022")}#Red frame

                                                                                #Visible border for rarity layout
                                                                                overlay_slot:
                                                                                    VerticalAlignment:=vertical_alignment.Center
                                                                                    HorizontalAlignment:=horizontal_alignment.Center
                                                                                    Padding:=margin{Left:=0.0, Top:=0.0, Right:=0.0, Bottom:=0.0}
                                                                                    Widget:=texture_block{DefaultImage:=Textures.General.T_GenericBorderSquare3, DefaultDesiredSize:=vector2{X:=255.0, Y:=40.0}, DefaultTint:=MakeColorFromHex("#2e7be3FF")}

                                                                                #Horizontal Stack box for rarity slots
                                                                                overlay_slot:
                                                                                    VerticalAlignment:=vertical_alignment.Fill
                                                                                    HorizontalAlignment:=horizontal_alignment.Fill
                                                                                    Padding:=margin{Left:=5.0, Top:=0.0, Right:=5.0, Bottom:=0.0}
                                                                                    Widget:=stack_box:
                                                                                        Orientation:=orientation.Horizontal
                                                                                        Slots:=array:
                                                                                            #Overlay for each rarity slots, including the pointing arrow
                                                                                            stack_box_slot:
                                                                                                VerticalAlignment:=vertical_alignment.Center
                                                                                                HorizontalAlignment:=horizontal_alignment.Center
                                                                                                Distribution:=option{1.0}
                                                                                                Padding:=margin{Left:=0.0, Top:=0.0, Right:=0.0, Bottom:=0.0}
                                                                                                Widget:=overlay:
                                                                                                    Slots:=array:
                                                                                                        #Makeshift size box
                                                                                                        overlay_slot:
                                                                                                            VerticalAlignment:=vertical_alignment.Center
                                                                                                            HorizontalAlignment:=horizontal_alignment.Center
                                                                                                            Padding:=margin{Left:=0.0, Top:=0.0, Right:=0.0, Bottom:=0.0}
                                                                                                            Widget:=texture_block{DefaultImage:=Textures.General.T_GenSquTrans, DefaultDesiredSize:=vector2{X:=30.0, Y:=40.0}, DefaultTint:=MakeColorFromHex("#0000FF08")}

                                                                                                        #Overlay for rarity border and texture
                                                                                                        overlay_slot:
                                                                                                            VerticalAlignment:=vertical_alignment.Center
                                                                                                            HorizontalAlignment:=horizontal_alignment.Center
                                                                                                            Padding:=margin{Left:=0.0, Top:=0.0, Right:=0.0, Bottom:=0.0}
                                                                                                            Widget:=overlay:
                                                                                                                Slots:=array:
                                                                                                                    
                                                                                                                    #Rarity texture border, when the rarity slot is empty
                                                                                                                    #overlay_slot:
                                                                                                                        #VerticalAlignment:=vertical_alignment.Center
                                                                                                                        #HorizontalAlignment:=horizontal_alignment.Center
                                                                                                                        #Padding:=margin{Left:=0.0, Top:=0.0, Right:=0.0, Bottom:=0.0}
                                                                                                                        #Widget:=texture_block{DefaultImage:=Textures.General.T_GenericBorderSquare2, DefaultDesiredSize:=vector2{X:=30.0, Y:=30.0}, DefaultTint:=MakeColorFromHex("#2e7be388")}

                                                                                                                    #Rarity texture
                                                                                                                    overlay_slot:
                                                                                                                        VerticalAlignment:=vertical_alignment.Center
                                                                                                                        HorizontalAlignment:=horizontal_alignment.Center
                                                                                                                        Padding:=margin{Left:=0.0, Top:=0.0, Right:=0.0, Bottom:=0.0}
                                                                                                                        Widget:=texture_block{DefaultImage:=Textures.General.T_GenericSquare, DefaultDesiredSize:=vector2{X:=20.0, Y:=20.0}, DefaultTint:=MakeColorFromHex("#979797FF")}
                                                                                                        
                                                                                                                    

                                                                                                        #Bracket
                                                                                                        overlay_slot:
                                                                                                            VerticalAlignment:=vertical_alignment.Fill
                                                                                                            HorizontalAlignment:=horizontal_alignment.Fill
                                                                                                            Padding:=margin{Left:=0.0, Top:=0.0, Right:=0.0, Bottom:=0.0}
                                                                                                            Widget:=stack_box:
                                                                                                                Orientation:=orientation.Horizontal
                                                                                                                #two instances of bracket texture to represent the range of rarities
                                                                                                                Slots:=array:

                                                                                                                    stack_box_slot:
                                                                                                                        VerticalAlignment:=vertical_alignment.Center
                                                                                                                        HorizontalAlignment:=horizontal_alignment.Left
                                                                                                                        Distribution:=option{1.0}
                                                                                                                        Padding:=margin{Left:=0.0, Top:=0.0, Right:=0.0, Bottom:=0.0}
                                                                                                                        Widget:=texture_block{DefaultImage:=Textures.General.T_RarityBracketLeft, DefaultDesiredSize:=vector2{X:=18.0, Y:=36.0}, DefaultTint:=MakeColorFromHex("#FFFFFF44")}

                                                                                                                    stack_box_slot:
                                                                                                                        VerticalAlignment:=vertical_alignment.Center
                                                                                                                        HorizontalAlignment:=horizontal_alignment.Right
                                                                                                                        Distribution:=option{1.0}
                                                                                                                        Padding:=margin{Left:=0.0, Top:=0.0, Right:=0.0, Bottom:=0.0}
                                                                                                                        Widget:=texture_block{DefaultImage:=Textures.General.T_RarityBracketRight, DefaultDesiredSize:=vector2{X:=18.0, Y:=36.0}, DefaultTint:=MakeColorFromHex("#FFFFFF44")}

                                                                                            stack_box_slot:
                                                                                                VerticalAlignment:=vertical_alignment.Center
                                                                                                HorizontalAlignment:=horizontal_alignment.Center
                                                                                                Distribution:=option{1.0}
                                                                                                Padding:=margin{Left:=0.0, Top:=0.0, Right:=0.0, Bottom:=0.0}
                                                                                                Widget:=overlay:
                                                                                                    Slots:=array:
                                                                                                        #Makeshift size box
                                                                                                        overlay_slot:
                                                                                                            VerticalAlignment:=vertical_alignment.Center
                                                                                                            HorizontalAlignment:=horizontal_alignment.Center
                                                                                                            Padding:=margin{Left:=0.0, Top:=0.0, Right:=0.0, Bottom:=0.0}
                                                                                                            Widget:=texture_block{DefaultImage:=Textures.General.T_GenSquTrans, DefaultDesiredSize:=vector2{X:=30.0, Y:=40.0}, DefaultTint:=MakeColorFromHex("#0000FF08")}

                                                                                                        #Overlay for rarity border and texture
                                                                                                        overlay_slot:
                                                                                                            VerticalAlignment:=vertical_alignment.Center
                                                                                                            HorizontalAlignment:=horizontal_alignment.Center
                                                                                                            Padding:=margin{Left:=0.0, Top:=0.0, Right:=0.0, Bottom:=0.0}
                                                                                                            Widget:=overlay:
                                                                                                                Slots:=array:
                                                                                                                    
                                                                                                                    #Rarity texture border, when the rarity slot is empty
                                                                                                                    overlay_slot:
                                                                                                                        VerticalAlignment:=vertical_alignment.Center
                                                                                                                        HorizontalAlignment:=horizontal_alignment.Center
                                                                                                                        Padding:=margin{Left:=0.0, Top:=0.0, Right:=0.0, Bottom:=0.0}
                                                                                                                        Widget:=texture_block{DefaultImage:=Textures.General.T_GenericBorderSquare2, DefaultDesiredSize:=vector2{X:=30.0, Y:=30.0}, DefaultTint:=MakeColorFromHex("#2e7be388")}

                                                                                                                    #Rarity texture
                                                                                                                    #overlay_slot:
                                                                                                                        #VerticalAlignment:=vertical_alignment.Center
                                                                                                                        #HorizontalAlignment:=horizontal_alignment.Center
                                                                                                                        #Padding:=margin{Left:=0.0, Top:=0.0, Right:=0.0, Bottom:=0.0}
                                                                                                                        #Widget:=texture_block{DefaultImage:=Textures.General.T_GenericSquare, DefaultDesiredSize:=vector2{X:=30.0, Y:=30.0}, DefaultTint:=MakeColorFromHex("#4ccc2aFF")}

                                                                                                        #Bracket
                                                                                                        #overlay_slot:
                                                                                                            #VerticalAlignment:=vertical_alignment.Fill
                                                                                                            #HorizontalAlignment:=horizontal_alignment.Fill
                                                                                                            #Padding:=margin{Left:=0.0, Top:=0.0, Right:=0.0, Bottom:=0.0}
                                                                                                            #Widget:=stack_box:
                                                                                                                #Orientation:=orientation.Horizontal
                                                                                                                #two instances of bracket texture to represent the range of rarities
                                                                                                                #Slots:=array:

                                                                                                                    #stack_box_slot:
                                                                                                                        #VerticalAlignment:=vertical_alignment.Center
                                                                                                                        #HorizontalAlignment:=horizontal_alignment.Left
                                                                                                                        #Distribution:=option{1.0}
                                                                                                                        #Padding:=margin{Left:=0.0, Top:=0.0, Right:=0.0, Bottom:=0.0}
                                                                                                                        #Widget:=texture_block{DefaultImage:=Textures.General.T_RarityBracketLeft, DefaultDesiredSize:=vector2{X:=18.0, Y:=36.0}, DefaultTint:=MakeColorFromHex("#FFFFFF44")}

                                                                                                                    #stack_box_slot:
                                                                                                                        #VerticalAlignment:=vertical_alignment.Center
                                                                                                                        #HorizontalAlignment:=horizontal_alignment.Right
                                                                                                                        #Distribution:=option{1.0}
                                                                                                                        #Padding:=margin{Left:=0.0, Top:=0.0, Right:=0.0, Bottom:=0.0}
                                                                                                                        #Widget:=texture_block{DefaultImage:=Textures.General.T_RarityBracketRight, DefaultDesiredSize:=vector2{X:=18.0, Y:=36.0}, DefaultTint:=MakeColorFromHex("#FFFFFF44")}

                                                                                            stack_box_slot:
                                                                                                VerticalAlignment:=vertical_alignment.Center
                                                                                                HorizontalAlignment:=horizontal_alignment.Center
                                                                                                Distribution:=option{1.0}
                                                                                                Padding:=margin{Left:=0.0, Top:=0.0, Right:=0.0, Bottom:=0.0}
                                                                                                Widget:=overlay:
                                                                                                    Slots:=array:
                                                                                                        #Makeshift size box
                                                                                                        overlay_slot:
                                                                                                            VerticalAlignment:=vertical_alignment.Center
                                                                                                            HorizontalAlignment:=horizontal_alignment.Center
                                                                                                            Padding:=margin{Left:=0.0, Top:=0.0, Right:=0.0, Bottom:=0.0}
                                                                                                            Widget:=texture_block{DefaultImage:=Textures.General.T_GenSquTrans, DefaultDesiredSize:=vector2{X:=30.0, Y:=40.0}, DefaultTint:=MakeColorFromHex("#0000FF08")}

                                                                                                        #Overlay for rarity border and texture
                                                                                                        overlay_slot:
                                                                                                            VerticalAlignment:=vertical_alignment.Center
                                                                                                            HorizontalAlignment:=horizontal_alignment.Center
                                                                                                            Padding:=margin{Left:=0.0, Top:=0.0, Right:=0.0, Bottom:=0.0}
                                                                                                            Widget:=overlay:
                                                                                                                Slots:=array:
                                                                                                                                                                                                                
                                                                                                                    #Rarity texture border, when the rarity slot is empty
                                                                                                                    overlay_slot:
                                                                                                                        VerticalAlignment:=vertical_alignment.Center
                                                                                                                        HorizontalAlignment:=horizontal_alignment.Center
                                                                                                                        Padding:=margin{Left:=0.0, Top:=0.0, Right:=0.0, Bottom:=0.0}
                                                                                                                        Widget:=texture_block{DefaultImage:=Textures.General.T_GenericBorderSquare2, DefaultDesiredSize:=vector2{X:=30.0, Y:=30.0}, DefaultTint:=MakeColorFromHex("#2e7be388")}
                                                                                                                                                                                                                            
                                                                                                                    #Rarity texture
                                                                                                                    #overlay_slot:
                                                                                                                        #VerticalAlignment:=vertical_alignment.Bottom
                                                                                                                        #HorizontalAlignment:=horizontal_alignment.Center
                                                                                                                        #Padding:=margin{Left:=0.0, Top:=0.0, Right:=0.0, Bottom:=0.0}
                                                                                                                        #Widget:=texture_block{DefaultImage:=Textures.General.T_GenericSquare, DefaultDesiredSize:=vector2{X:=30.0, Y:=30.0}, DefaultTint:=MakeColorFromHex("#2a54ccFF")}
                                                                                                                    
                                                                                                        #Bracket
                                                                                                        #overlay_slot:
                                                                                                            #VerticalAlignment:=vertical_alignment.Fill
                                                                                                            #HorizontalAlignment:=horizontal_alignment.Fill
                                                                                                            #Padding:=margin{Left:=0.0, Top:=0.0, Right:=0.0, Bottom:=0.0}
                                                                                                            #Widget:=stack_box:
                                                                                                                #Orientation:=orientation.Horizontal
                                                                                                                #two instances of bracket texture to represent the range of rarities
                                                                                                                #Slots:=array:

                                                                                                                    #stack_box_slot:
                                                                                                                        #VerticalAlignment:=vertical_alignment.Center
                                                                                                                        #HorizontalAlignment:=horizontal_alignment.Left
                                                                                                                        #Distribution:=option{1.0}
                                                                                                                        #Padding:=margin{Left:=0.0, Top:=0.0, Right:=0.0, Bottom:=0.0}
                                                                                                                        #Widget:=texture_block{DefaultImage:=Textures.General.T_RarityBracketLeft, DefaultDesiredSize:=vector2{X:=18.0, Y:=36.0}, DefaultTint:=MakeColorFromHex("#FFFFFF44")}

                                                                                                                    #stack_box_slot:
                                                                                                                        #VerticalAlignment:=vertical_alignment.Center
                                                                                                                        #HorizontalAlignment:=horizontal_alignment.Right
                                                                                                                        #Distribution:=option{1.0}
                                                                                                                        #Padding:=margin{Left:=0.0, Top:=0.0, Right:=0.0, Bottom:=0.0}
                                                                                                                        #Widget:=texture_block{DefaultImage:=Textures.General.T_RarityBracketRight, DefaultDesiredSize:=vector2{X:=18.0, Y:=36.0}, DefaultTint:=MakeColorFromHex("#FFFFFF44")}

                                                                                            stack_box_slot:
                                                                                                VerticalAlignment:=vertical_alignment.Center
                                                                                                HorizontalAlignment:=horizontal_alignment.Center
                                                                                                Distribution:=option{1.0}
                                                                                                Padding:=margin{Left:=0.0, Top:=0.0, Right:=0.0, Bottom:=0.0}
                                                                                                Widget:=overlay:
                                                                                                    Slots:=array:
                                                                                                        #Makeshift size box
                                                                                                        overlay_slot:
                                                                                                            VerticalAlignment:=vertical_alignment.Center
                                                                                                            HorizontalAlignment:=horizontal_alignment.Center
                                                                                                            Padding:=margin{Left:=0.0, Top:=0.0, Right:=0.0, Bottom:=0.0}
                                                                                                            Widget:=texture_block{DefaultImage:=Textures.General.T_GenSquTrans, DefaultDesiredSize:=vector2{X:=30.0, Y:=40.0}, DefaultTint:=MakeColorFromHex("#0000FF08")}

                                                                                                        #Overlay for rarity border and texture
                                                                                                        overlay_slot:
                                                                                                            VerticalAlignment:=vertical_alignment.Center
                                                                                                            HorizontalAlignment:=horizontal_alignment.Center
                                                                                                            Padding:=margin{Left:=0.0, Top:=0.0, Right:=0.0, Bottom:=0.0}
                                                                                                            Widget:=overlay:
                                                                                                                Slots:=array:

                                                                                                                    #Rarity texture border, when the rarity slot is empty
                                                                                                                    overlay_slot:
                                                                                                                        VerticalAlignment:=vertical_alignment.Center
                                                                                                                        HorizontalAlignment:=horizontal_alignment.Center
                                                                                                                        Padding:=margin{Left:=0.0, Top:=0.0, Right:=0.0, Bottom:=0.0}
                                                                                                                        Widget:=texture_block{DefaultImage:=Textures.General.T_GenericBorderSquare2, DefaultDesiredSize:=vector2{X:=30.0, Y:=30.0}, DefaultTint:=MakeColorFromHex("#2e7be388")}
                                                                                                                    
                                                                                                                    #Rarity texture
                                                                                                                    #overlay_slot:
                                                                                                                        #VerticalAlignment:=vertical_alignment.Bottom
                                                                                                                        #HorizontalAlignment:=horizontal_alignment.Center
                                                                                                                        #Padding:=margin{Left:=0.0, Top:=0.0, Right:=0.0, Bottom:=0.0}
                                                                                                                        #Widget:=texture_block{DefaultImage:=Textures.General.T_GenericSquare, DefaultDesiredSize:=vector2{X:=30.0, Y:=30.0}, DefaultTint:=MakeColorFromHex("#b32accFF")}
                                                                                                                    
                                                                                                        #Bracket
                                                                                                        #overlay_slot:
                                                                                                            #VerticalAlignment:=vertical_alignment.Fill
                                                                                                            #HorizontalAlignment:=horizontal_alignment.Fill
                                                                                                            #Padding:=margin{Left:=0.0, Top:=0.0, Right:=0.0, Bottom:=0.0}
                                                                                                            #Widget:=stack_box:
                                                                                                                #Orientation:=orientation.Horizontal
                                                                                                                #two instances of bracket texture to represent the range of rarities
                                                                                                                #Slots:=array:

                                                                                                                    #stack_box_slot:
                                                                                                                        #VerticalAlignment:=vertical_alignment.Center
                                                                                                                        #HorizontalAlignment:=horizontal_alignment.Left
                                                                                                                        #Distribution:=option{1.0}
                                                                                                                        #Padding:=margin{Left:=0.0, Top:=0.0, Right:=0.0, Bottom:=0.0}
                                                                                                                        #Widget:=texture_block{DefaultImage:=Textures.General.T_RarityBracketLeft, DefaultDesiredSize:=vector2{X:=18.0, Y:=36.0}, DefaultTint:=MakeColorFromHex("#FFFFFF44")}

                                                                                                                    #stack_box_slot:
                                                                                                                        #VerticalAlignment:=vertical_alignment.Center
                                                                                                                        #HorizontalAlignment:=horizontal_alignment.Right
                                                                                                                        #Distribution:=option{1.0}
                                                                                                                        #Padding:=margin{Left:=0.0, Top:=0.0, Right:=0.0, Bottom:=0.0}
                                                                                                                        #Widget:=texture_block{DefaultImage:=Textures.General.T_RarityBracketRight, DefaultDesiredSize:=vector2{X:=18.0, Y:=36.0}, DefaultTint:=MakeColorFromHex("#FFFFFF44")}

                                                                                            stack_box_slot:
                                                                                                VerticalAlignment:=vertical_alignment.Center
                                                                                                HorizontalAlignment:=horizontal_alignment.Center
                                                                                                Distribution:=option{1.0}
                                                                                                Padding:=margin{Left:=0.0, Top:=0.0, Right:=0.0, Bottom:=0.0}
                                                                                                Widget:=overlay:
                                                                                                    Slots:=array:
                                                                                                        #Makeshift size box
                                                                                                        overlay_slot:
                                                                                                            VerticalAlignment:=vertical_alignment.Center
                                                                                                            HorizontalAlignment:=horizontal_alignment.Center
                                                                                                            Padding:=margin{Left:=0.0, Top:=0.0, Right:=0.0, Bottom:=0.0}
                                                                                                            Widget:=texture_block{DefaultImage:=Textures.General.T_GenSquTrans, DefaultDesiredSize:=vector2{X:=30.0, Y:=40.0}, DefaultTint:=MakeColorFromHex("#0000FF08")}

                                                                                                        #Overlay for rarity border and texture
                                                                                                        overlay_slot:
                                                                                                            VerticalAlignment:=vertical_alignment.Center
                                                                                                            HorizontalAlignment:=horizontal_alignment.Center
                                                                                                            Padding:=margin{Left:=0.0, Top:=0.0, Right:=0.0, Bottom:=0.0}
                                                                                                            Widget:=overlay:
                                                                                                                Slots:=array:

                                                                                                                    #Rarity texture border, when the rarity slot is empty
                                                                                                                    overlay_slot:
                                                                                                                        VerticalAlignment:=vertical_alignment.Center
                                                                                                                        HorizontalAlignment:=horizontal_alignment.Center
                                                                                                                        Padding:=margin{Left:=0.0, Top:=0.0, Right:=0.0, Bottom:=0.0}
                                                                                                                        Widget:=texture_block{DefaultImage:=Textures.General.T_GenericBorderSquare2, DefaultDesiredSize:=vector2{X:=30.0, Y:=30.0}, DefaultTint:=MakeColorFromHex("#2e7be388")}
                                                                                                                    
                                                                                                                    #Rarity texture
                                                                                                                    #overlay_slot:
                                                                                                                        #VerticalAlignment:=vertical_alignment.Bottom
                                                                                                                        #HorizontalAlignment:=horizontal_alignment.Center
                                                                                                                        #Padding:=margin{Left:=0.0, Top:=0.0, Right:=0.0, Bottom:=0.0}
                                                                                                                        #Widget:=texture_block{DefaultImage:=Textures.General.T_GenericSquare, DefaultDesiredSize:=vector2{X:=30.0, Y:=30.0}, DefaultTint:=MakeColorFromHex("#cc8d2aFF")}

                                                                                                        #Bracket
                                                                                                        #overlay_slot:
                                                                                                            #VerticalAlignment:=vertical_alignment.Fill
                                                                                                            #HorizontalAlignment:=horizontal_alignment.Fill
                                                                                                            #Padding:=margin{Left:=0.0, Top:=0.0, Right:=0.0, Bottom:=0.0}
                                                                                                            #Widget:=stack_box:
                                                                                                                #Orientation:=orientation.Horizontal
                                                                                                                #two instances of bracket texture to represent the range of rarities
                                                                                                                #Slots:=array:

                                                                                                                    #stack_box_slot:
                                                                                                                        #VerticalAlignment:=vertical_alignment.Center
                                                                                                                        #HorizontalAlignment:=horizontal_alignment.Left
                                                                                                                        #Distribution:=option{1.0}
                                                                                                                        #Padding:=margin{Left:=0.0, Top:=0.0, Right:=0.0, Bottom:=0.0}
                                                                                                                        #Widget:=texture_block{DefaultImage:=Textures.General.T_RarityBracketLeft, DefaultDesiredSize:=vector2{X:=18.0, Y:=36.0}, DefaultTint:=MakeColorFromHex("#FFFFFF44")}

                                                                                                                    #stack_box_slot:
                                                                                                                        #VerticalAlignment:=vertical_alignment.Center
                                                                                                                        #HorizontalAlignment:=horizontal_alignment.Right
                                                                                                                        #Distribution:=option{1.0}
                                                                                                                        #Padding:=margin{Left:=0.0, Top:=0.0, Right:=0.0, Bottom:=0.0}
                                                                                                                        #Widget:=texture_block{DefaultImage:=Textures.General.T_RarityBracketRight, DefaultDesiredSize:=vector2{X:=18.0, Y:=36.0}, DefaultTint:=MakeColorFromHex("#FFFFFF44")}

                                                                                            stack_box_slot:
                                                                                                VerticalAlignment:=vertical_alignment.Center
                                                                                                HorizontalAlignment:=horizontal_alignment.Center
                                                                                                Distribution:=option{1.0}
                                                                                                Padding:=margin{Left:=0.0, Top:=0.0, Right:=0.0, Bottom:=0.0}
                                                                                                Widget:=overlay:
                                                                                                    Slots:=array:
                                                                                                        #Makeshift size box
                                                                                                        overlay_slot:
                                                                                                            VerticalAlignment:=vertical_alignment.Center
                                                                                                            HorizontalAlignment:=horizontal_alignment.Center
                                                                                                            Padding:=margin{Left:=0.0, Top:=0.0, Right:=0.0, Bottom:=0.0}
                                                                                                            Widget:=texture_block{DefaultImage:=Textures.General.T_GenSquTrans, DefaultDesiredSize:=vector2{X:=30.0, Y:=40.0}, DefaultTint:=MakeColorFromHex("#0000FF08")}

                                                                                                        #Overlay for rarity border and texture
                                                                                                        overlay_slot:
                                                                                                            VerticalAlignment:=vertical_alignment.Center
                                                                                                            HorizontalAlignment:=horizontal_alignment.Center
                                                                                                            Padding:=margin{Left:=0.0, Top:=0.0, Right:=0.0, Bottom:=0.0}
                                                                                                            Widget:=overlay:
                                                                                                                Slots:=array:
                                                                                                        
                                                                                                                    #Rarity texture border, when the rarity slot is empty
                                                                                                                    overlay_slot:
                                                                                                                        VerticalAlignment:=vertical_alignment.Center
                                                                                                                        HorizontalAlignment:=horizontal_alignment.Center
                                                                                                                        Padding:=margin{Left:=0.0, Top:=0.0, Right:=0.0, Bottom:=0.0}
                                                                                                                        Widget:=texture_block{DefaultImage:=Textures.General.T_GenericBorderSquare2, DefaultDesiredSize:=vector2{X:=30.0, Y:=30.0}, DefaultTint:=MakeColorFromHex("#2e7be388")}
                                                                                                                    
                                                                                                                    #Rarity texture
                                                                                                                    #overlay_slot:
                                                                                                                        #VerticalAlignment:=vertical_alignment.Bottom
                                                                                                                        #HorizontalAlignment:=horizontal_alignment.Center
                                                                                                                        #Padding:=margin{Left:=0.0, Top:=0.0, Right:=0.0, Bottom:=0.0}
                                                                                                                        #Widget:=texture_block{DefaultImage:=Textures.General.T_GenericSquare, DefaultDesiredSize:=vector2{X:=30.0, Y:=30.0}, DefaultTint:=MakeColorFromHex("#eacf19FF")}
                                                                                                                    
                                                                                                        #Bracket
                                                                                                        #overlay_slot:
                                                                                                            #VerticalAlignment:=vertical_alignment.Fill
                                                                                                            #HorizontalAlignment:=horizontal_alignment.Fill
                                                                                                            #Padding:=margin{Left:=0.0, Top:=0.0, Right:=0.0, Bottom:=0.0}
                                                                                                            #Widget:=stack_box:
                                                                                                                #Orientation:=orientation.Horizontal
                                                                                                                #two instances of bracket texture to represent the range of rarities
                                                                                                                #Slots:=array:

                                                                                                                    #stack_box_slot:
                                                                                                                        #VerticalAlignment:=vertical_alignment.Center
                                                                                                                        #HorizontalAlignment:=horizontal_alignment.Left
                                                                                                                        #Distribution:=option{1.0}
                                                                                                                        #Padding:=margin{Left:=0.0, Top:=0.0, Right:=0.0, Bottom:=0.0}
                                                                                                                        #Widget:=texture_block{DefaultImage:=Textures.General.T_RarityBracketLeft, DefaultDesiredSize:=vector2{X:=18.0, Y:=36.0}, DefaultTint:=MakeColorFromHex("#FFFFFF44")}

                                                                                                                    #stack_box_slot:
                                                                                                                        #VerticalAlignment:=vertical_alignment.Center
                                                                                                                        #HorizontalAlignment:=horizontal_alignment.Right
                                                                                                                        #Distribution:=option{1.0}
                                                                                                                        #Padding:=margin{Left:=0.0, Top:=0.0, Right:=0.0, Bottom:=0.0}
                                                                                                                        #Widget:=texture_block{DefaultImage:=Textures.General.T_RarityBracketRight, DefaultDesiredSize:=vector2{X:=18.0, Y:=36.0}, DefaultTint:=MakeColorFromHex("#FFFFFF44")}

                                            #Upgrade Button
                                            stack_box_slot:
                                                VerticalAlignment:=vertical_alignment.Center
                                                HorizontalAlignment:=horizontal_alignment.Center
                                                Padding:=margin{Left:=0.0, Top:=0.0, Right:=0.0, Bottom:=0.0}
                                                Widget:=button_loud{DefaultText:=StringToMessage("UPDATE\n$15")}

                    #overlay_slot:
                        #VerticalAlignment:=vertical_alignment.Fill
                        #HorizontalAlignment:=horizontal_alignment.Fill
                        #Padding:=margin{Left:=0.0, Top:=0.0, Right:=0.0, Bottom:=0.0}
                        #Widget:=texture_block{DefaultImage:=Textures.General.T_LevelPointer, DefaultDesiredSize:=vector2{X:=20.0, Y:=20.0}, DefaultTint:=MakeColorFromHex("#7de2FFFF")}
                        
            LevelRow3:=overlay:
                Slots:=array:
                    overlay_slot:
                        #Makeshift size box
                        VerticalAlignment:=vertical_alignment.Center
                        HorizontalAlignment:=horizontal_alignment.Center
                        Padding:=margin{Left:=0.0, Top:=0.0, Right:=0.0, Bottom:=0.0}
                        Widget:=texture_block{DefaultImage:=Textures.General.T_GenSquTrans, DefaultDesiredSize:=vector2{X:=850.0, Y:=130.0}, DefaultTint:=MakeColorFromHex("#0000FFFF")}

                    overlay_slot:
                        #Rarity level content
                        VerticalAlignment:=vertical_alignment.Center
                        HorizontalAlignment:=horizontal_alignment.Left
                        Padding:=margin{Left:=0.0, Top:=0.0, Right:=0.0, Bottom:=0.0}
                        Widget:=overlay:
                            Slots:=array:

                                overlay_slot:
                                    #Makeshift size box
                                    VerticalAlignment:=vertical_alignment.Center
                                    HorizontalAlignment:=horizontal_alignment.Center
                                    Padding:=margin{Left:=0.0, Top:=0.0, Right:=0.0, Bottom:=0.0}
                                    Widget:=texture_block{DefaultImage:=Textures.General.T_GenericSquare, DefaultDesiredSize:=vector2{X:=850.0, Y:=130.0}, DefaultTint:=MakeColorFromHex("#010a1832")}

                                overlay_slot:
                                    #Highlight texture, over the content but under the rarity flag
                                    VerticalAlignment:=vertical_alignment.Fill
                                    HorizontalAlignment:=horizontal_alignment.Fill
                                    Padding:=margin{Left:=0.0, Top:=0.0, Right:=0.0, Bottom:=0.0}
                                    Widget:=texture_block{DefaultImage:=Textures.General.T_LevelBorder, DefaultDesiredSize:=vector2{X:=512.0, Y:=64.0}, DefaultTint:=MakeColorFromHex("#1d4b89FF")}

                                    #Horizontal stack box
                                overlay_slot:
                                    VerticalAlignment:=vertical_alignment.Fill
                                    HorizontalAlignment:=horizontal_alignment.Left
                                    Padding:=margin{Left:=0.0, Top:=0.0, Right:=0.0, Bottom:=0.0}
                                    Widget:=stack_box:
                                        Orientation:=orientation.Horizontal
                                        Slots:=array:
                                            #Level flag
                                            stack_box_slot:
                                                VerticalAlignment:=vertical_alignment.Top
                                                HorizontalAlignment:=horizontal_alignment.Left
                                                Padding:=margin{Left:=0.0, Top:=0.0, Right:=15.0, Bottom:=0.0}
                                                Widget:=overlay:

                                                    Slots:=array:
                                                        overlay_slot:
                                                            VerticalAlignment:=vertical_alignment.Center
                                                            HorizontalAlignment:=horizontal_alignment.Center
                                                            Padding:=margin{Left:=0.0, Top:=0.0, Right:=0.0, Bottom:=0.0}
                                                            #Tier level makeshift size box
                                                            Widget:=texture_block{DefaultImage:=Textures.General.T_GenSquTrans, DefaultDesiredSize:=vector2{X:=150.0, Y:=130.0}, DefaultTint:=MakeColorFromHex("#FF0000FF")}

                                                        overlay_slot:
                                                            VerticalAlignment:=vertical_alignment.Top
                                                            HorizontalAlignment:=horizontal_alignment.Left
                                                            Padding:=margin{Left:=0.0, Top:=0.0, Right:=0.0, Bottom:=0.0}
                                                            #Selected level yellow flag
                                                            Widget:=stack_box:
                                                                Orientation:=orientation.Vertical
                                                                Slots:=array:
                                                                    stack_box_slot:
                                                                        VerticalAlignment:=vertical_alignment.Center
                                                                        HorizontalAlignment:=horizontal_alignment.Center
                                                                        Padding:=margin{Left:=0.0, Top:=0.0, Right:=0.0, Bottom:=0.0}
                                                                        #level flag texture
                                                                        Widget:=overlay:
                                                                            Slots:=array:

                                                                                overlay_slot:
                                                                                    VerticalAlignment:=vertical_alignment.Center
                                                                                    HorizontalAlignment:=horizontal_alignment.Center
                                                                                    Padding:=margin{Left:=0.0, Top:=0.0, Right:=0.0, Bottom:=0.0}
                                                                                    #level flag texture
                                                                                    Widget:=texture_block{DefaultImage:=Textures.General.T_GenLevelFlag, DefaultDesiredSize:=vector2{X:=150.0, Y:=45.0}, DefaultTint:=MakeColorFromHex("#f7ff1aFF")}

                                                                                overlay_slot:
                                                                                    VerticalAlignment:=vertical_alignment.Center
                                                                                    HorizontalAlignment:=horizontal_alignment.Center
                                                                                    Padding:=margin{Left:=0.0, Top:=0.0, Right:=0.0, Bottom:=0.0}
                                                                                    #level flag text
                                                                                    Widget:=text_block{DefaultText := StringToMessage("Base LV"), DefaultTextColor:=MakeColorFromHex("#000000FF")}

                                                                    stack_box_slot:
                                                                        VerticalAlignment:=vertical_alignment.Center
                                                                        HorizontalAlignment:=horizontal_alignment.Center
                                                                        Padding:=margin{Left:=0.0, Top:=0.0, Right:=0.0, Bottom:=0.0}
                                                                        #level flag texture
                                                                        Widget:=overlay:
                                                                            Slots:=array:

                                                                                #Booster Icon - Increase Max Shield
                                                                                overlay_slot:
                                                                                    VerticalAlignment:=vertical_alignment.Center
                                                                                    HorizontalAlignment:=horizontal_alignment.Center
                                                                                    Padding:=margin{Left:=0.0, Top:=0.0, Right:=0.0, Bottom:=0.0}
                                                                                    #Weapon icon BG
                                                                                    Widget:=texture_block{DefaultImage:=Textures.Rarity.T_IconBG_Common, DefaultDesiredSize:=vector2{X:=78.0, Y:=78.0}, DefaultTint:=MakeColorFromHex("#FFFFFFFF")}
                        
                                                                                overlay_slot:
                                                                                    VerticalAlignment:=vertical_alignment.Center
                                                                                    HorizontalAlignment:=horizontal_alignment.Center
                                                                                    Padding:=margin{Left:=0.0, Top:=0.0, Right:=0.0, Bottom:=0.0}
                                                                                    #Weapon icon Border
                                                                                    Widget:=texture_block{DefaultImage:=Textures.General.T_BoosterFrame, DefaultDesiredSize:=vector2{X:=78.0, Y:=78.0}, DefaultTint:=MakeColorFromHex("#FFFFFFFF")}
                        
                                                                                overlay_slot:
                                                                                    VerticalAlignment:=vertical_alignment.Center
                                                                                    HorizontalAlignment:=horizontal_alignment.Center
                                                                                    Padding:=margin{Left:=0.0, Top:=0.0, Right:=0.0, Bottom:=0.0}
                                                                                    #Weapon icon - pistol
                                                                                    Widget:=texture_block{DefaultImage:=Textures.Boosters.T_IncreaseMaxShield, DefaultDesiredSize:=vector2{X:=70.0, Y:=70.0}, DefaultTint:=MakeColorFromHex("#FFFFFFFF")}
                        
                                                                                #overlay_slot:
                                                                                    #VerticalAlignment:=vertical_alignment.Bottom
                                                                                    #HorizontalAlignment:=horizontal_alignment.Center
                                                                                    #Padding:=margin{Left:=0.0, Top:=0.0, Right:=0.0, Bottom:=5.0}
                                                                                    ##Weapon header - pistol
                                                                                    #Widget:=texture_block{DefaultImage:=Textures.Items.T_HeaderPistol, DefaultDesiredSize:=vector2{X:=78.0, Y:=18.0}, DefaultTint:=MakeColorFromHex("#FFFFFFFF")}

                                                        #overlay_slot:
                                                            #VerticalAlignment:=vertical_alignment.Fill
                                                            #HorizontalAlignment:=horizontal_alignment.Center
                                                            #Padding:=margin{Left:=0.0, Top:=5.0, Right:=0.0, Bottom:=5.0}
                                                            #Unselected level and lock
                                                            #Widget:=overlay:
                                                                #Slots:=array:
                                                                    
                                                                    #overlay_slot:
                                                                        #VerticalAlignment:=vertical_alignment.Top
                                                                        #HorizontalAlignment:=horizontal_alignment.Center
                                                                        #Padding:=margin{Left:=0.0, Top:=0.0, Right:=0.0, Bottom:=0.0}
                                                                        #level flag texture
                                                                        #Widget:=text_block{DefaultText := StringToMessage("Base Level"), DefaultTextColor:=MakeColorFromHex("#FFFFFFFF")}

                                                                    #overlay_slot:
                                                                        #VerticalAlignment:=vertical_alignment.Bottom
                                                                        #HorizontalAlignment:=horizontal_alignment.Center
                                                                        #Padding:=margin{Left:=0.0, Top:=0.0, Right:=0.0, Bottom:=0.0}
                                                                        ##level flag texture
                                                                        #Widget:=overlay:
                                                                            #Slots:=array:
                                                                                ##Weapon Icon - Pistols
                                                                                #overlay_slot:
                                                                                    #VerticalAlignment:=vertical_alignment.Center
                                                                                    #HorizontalAlignment:=horizontal_alignment.Center
                                                                                    #Padding:=margin{Left:=0.0, Top:=0.0, Right:=0.0, Bottom:=0.0}
                                                                                    ##Weapon icon BG
                                                                                    #Widget:=texture_block{DefaultImage:=Textures.General.T_GenericSquare, DefaultDesiredSize:=vector2{X:=78.0, Y:=78.0}, DefaultTint:=MakeColorFromHex("#000c1d88")}
                        
                                                                                #overlay_slot:
                                                                                    #VerticalAlignment:=vertical_alignment.Center
                                                                                    #HorizontalAlignment:=horizontal_alignment.Center
                                                                                    #Padding:=margin{Left:=0.0, Top:=0.0, Right:=0.0, Bottom:=0.0}
                                                                                    ##Weapon icon Border
                                                                                    #Widget:=texture_block{DefaultImage:=Textures.General.T_GenericBorderSquare, DefaultDesiredSize:=vector2{X:=78.0, Y:=78.0}, DefaultTint:=MakeColorFromHex("#2e7be3FF")}
                        
                                                                                #overlay_slot:
                                                                                    #VerticalAlignment:=vertical_alignment.Top
                                                                                    #HorizontalAlignment:=horizontal_alignment.Center
                                                                                    #Padding:=margin{Left:=0.0, Top:=5.0, Right:=0.0, Bottom:=0.0}
                                                                                    ##Weapon icon - pistol
                                                                                    #Widget:=texture_block{DefaultImage:=Textures.Items.T_IconPistol, DefaultDesiredSize:=vector2{X:=50.0, Y:=50.0}, DefaultTint:=MakeColorFromHex("#FFFFFFFF")}
                        
                                                                                #overlay_slot:
                                                                                    #VerticalAlignment:=vertical_alignment.Bottom
                                                                                    #HorizontalAlignment:=horizontal_alignment.Center
                                                                                    #Padding:=margin{Left:=0.0, Top:=0.0, Right:=0.0, Bottom:=5.0}
                                                                                    ##Weapon header - pistol
                                                                                    #Widget:=texture_block{DefaultImage:=Textures.Items.T_HeaderPistol, DefaultDesiredSize:=vector2{X:=78.0, Y:=18.0}, DefaultTint:=MakeColorFromHex("#FFFFFFFF")}
                                            
                                            

                                            stack_box_slot:
                                                VerticalAlignment:=vertical_alignment.Fill
                                                HorizontalAlignment:=horizontal_alignment.Center #Center
                                                Padding:=margin{Left:=0.0, Top:=0.0, Right:=0.0, Bottom:=0.0}
                                                Widget:=overlay:
                                                    Slots:=array:
                                                        overlay_slot:
                                                            #Makeshift size box
                                                            VerticalAlignment:=vertical_alignment.Center
                                                            HorizontalAlignment:=horizontal_alignment.Center
                                                            Padding:=margin{Left:=0.0, Top:=0.0, Right:=0.0, Bottom:=0.0}
                                                            Widget:=texture_block{DefaultImage:=Textures.General.T_GenSquTrans, DefaultDesiredSize:=vector2{X:=418.0, Y:=100.0}, DefaultTint:=MakeColorFromHex("#00FF0011")}

                                                        overlay_slot:
                                                            #Vertical stack box to add all the rarity information
                                                            VerticalAlignment:=vertical_alignment.Center
                                                            HorizontalAlignment:=horizontal_alignment.Center
                                                            Padding:=margin{Left:=0.0, Top:=0.0, Right:=0.0, Bottom:=0.0}
                                                            Widget:=stack_box:
                                                                Orientation:=orientation.Vertical
                                                                Slots:=array:
                                                                    stack_box_slot:
                                                                        #Horizontal Stack box for max rarity available
                                                                        VerticalAlignment:=vertical_alignment.Top
                                                                        HorizontalAlignment:=horizontal_alignment.Center
                                                                        Padding:=margin{Left:=0.0, Top:=0.0, Right:=0.0, Bottom:=0.0}
                                                                        Widget:=stack_box:
                                                                            Orientation:=orientation.Horizontal
                                                                            Slots:=array:
                                                                                stack_box_slot:
                                                                                    VerticalAlignment:=vertical_alignment.Center
                                                                                    HorizontalAlignment:=horizontal_alignment.Center
                                                                                    Padding:=margin{Left:=0.0, Top:=0.0, Right:=0.0, Bottom:=0.0}
                                                                                    Widget:=text_block{DefaultText := StringToMessage("Increase Max Shield"), DefaultTextColor:=MakeColorFromHex("#FFFFFFFF")}

                                                                    stack_box_slot:
                                                                        #Horizontal overlay for overall rarity scheme
                                                                        VerticalAlignment:=vertical_alignment.Bottom
                                                                        HorizontalAlignment:=horizontal_alignment.Center
                                                                        Padding:=margin{Left:=0.0, Top:=0.0, Right:=0.0, Bottom:=0.0}
                                                                        Widget:=overlay:
                                                                            Slots:=array:
                                                                                
                                                                                #Makeshift size box for the rarity scheme
                                                                                overlay_slot:
                                                                                    VerticalAlignment:=vertical_alignment.Center
                                                                                    HorizontalAlignment:=horizontal_alignment.Center
                                                                                    Padding:=margin{Left:=0.0, Top:=0.0, Right:=0.0, Bottom:=0.0}
                                                                                    Widget:=texture_block{DefaultImage:=Textures.General.T_GenSquTrans, DefaultDesiredSize:=vector2{X:=255.0, Y:=50.0}, DefaultTint:=MakeColorFromHex("#FF000022")}#Red frame

                                                                                #Visible border for rarity layout
                                                                                overlay_slot:
                                                                                    VerticalAlignment:=vertical_alignment.Center
                                                                                    HorizontalAlignment:=horizontal_alignment.Center
                                                                                    Padding:=margin{Left:=0.0, Top:=0.0, Right:=0.0, Bottom:=0.0}
                                                                                    Widget:=texture_block{DefaultImage:=Textures.General.T_GenericBorderSquare3, DefaultDesiredSize:=vector2{X:=255.0, Y:=40.0}, DefaultTint:=MakeColorFromHex("#2e7be3FF")}

                                                                                #Horizontal Stack box for rarity slots
                                                                                overlay_slot:
                                                                                    VerticalAlignment:=vertical_alignment.Fill
                                                                                    HorizontalAlignment:=horizontal_alignment.Fill
                                                                                    Padding:=margin{Left:=5.0, Top:=0.0, Right:=5.0, Bottom:=0.0}
                                                                                    Widget:=stack_box:
                                                                                        Orientation:=orientation.Horizontal
                                                                                        Slots:=array:
                                                                                            #Overlay for each rarity slots, including the pointing arrow
                                                                                            stack_box_slot:
                                                                                                VerticalAlignment:=vertical_alignment.Center
                                                                                                HorizontalAlignment:=horizontal_alignment.Center
                                                                                                Distribution:=option{1.0}
                                                                                                Padding:=margin{Left:=0.0, Top:=0.0, Right:=0.0, Bottom:=0.0}
                                                                                                Widget:=overlay:
                                                                                                    Slots:=array:
                                                                                                        #Makeshift size box
                                                                                                        overlay_slot:
                                                                                                            VerticalAlignment:=vertical_alignment.Center
                                                                                                            HorizontalAlignment:=horizontal_alignment.Center
                                                                                                            Padding:=margin{Left:=0.0, Top:=0.0, Right:=0.0, Bottom:=0.0}
                                                                                                            Widget:=texture_block{DefaultImage:=Textures.General.T_GenSquTrans, DefaultDesiredSize:=vector2{X:=30.0, Y:=40.0}, DefaultTint:=MakeColorFromHex("#0000FF08")}

                                                                                                        #Overlay for rarity border and texture
                                                                                                        overlay_slot:
                                                                                                            VerticalAlignment:=vertical_alignment.Center
                                                                                                            HorizontalAlignment:=horizontal_alignment.Center
                                                                                                            Padding:=margin{Left:=0.0, Top:=0.0, Right:=0.0, Bottom:=0.0}
                                                                                                            Widget:=overlay:
                                                                                                                Slots:=array:
                                                                                                                    
                                                                                                                    #Rarity texture border, when the rarity slot is empty
                                                                                                                    #overlay_slot:
                                                                                                                        #VerticalAlignment:=vertical_alignment.Center
                                                                                                                        #HorizontalAlignment:=horizontal_alignment.Center
                                                                                                                        #Padding:=margin{Left:=0.0, Top:=0.0, Right:=0.0, Bottom:=0.0}
                                                                                                                        #Widget:=texture_block{DefaultImage:=Textures.General.T_GenericBorderSquare2, DefaultDesiredSize:=vector2{X:=30.0, Y:=30.0}, DefaultTint:=MakeColorFromHex("#2e7be388")}

                                                                                                                    #Rarity texture
                                                                                                                    overlay_slot:
                                                                                                                        VerticalAlignment:=vertical_alignment.Center
                                                                                                                        HorizontalAlignment:=horizontal_alignment.Center
                                                                                                                        Padding:=margin{Left:=0.0, Top:=0.0, Right:=0.0, Bottom:=0.0}
                                                                                                                        Widget:=texture_block{DefaultImage:=Textures.General.T_GenericSquare, DefaultDesiredSize:=vector2{X:=20.0, Y:=20.0}, DefaultTint:=MakeColorFromHex("#979797FF")}
                                                                                                        
                                                                                                                    

                                                                                                        #Bracket
                                                                                                        overlay_slot:
                                                                                                            VerticalAlignment:=vertical_alignment.Fill
                                                                                                            HorizontalAlignment:=horizontal_alignment.Fill
                                                                                                            Padding:=margin{Left:=0.0, Top:=0.0, Right:=0.0, Bottom:=0.0}
                                                                                                            Widget:=stack_box:
                                                                                                                Orientation:=orientation.Horizontal
                                                                                                                #two instances of bracket texture to represent the range of rarities
                                                                                                                Slots:=array:

                                                                                                                    stack_box_slot:
                                                                                                                        VerticalAlignment:=vertical_alignment.Center
                                                                                                                        HorizontalAlignment:=horizontal_alignment.Left
                                                                                                                        Distribution:=option{1.0}
                                                                                                                        Padding:=margin{Left:=0.0, Top:=0.0, Right:=0.0, Bottom:=0.0}
                                                                                                                        Widget:=texture_block{DefaultImage:=Textures.General.T_RarityBracketLeft, DefaultDesiredSize:=vector2{X:=18.0, Y:=36.0}, DefaultTint:=MakeColorFromHex("#FFFFFF44")}

                                                                                                                    stack_box_slot:
                                                                                                                        VerticalAlignment:=vertical_alignment.Center
                                                                                                                        HorizontalAlignment:=horizontal_alignment.Right
                                                                                                                        Distribution:=option{1.0}
                                                                                                                        Padding:=margin{Left:=0.0, Top:=0.0, Right:=0.0, Bottom:=0.0}
                                                                                                                        Widget:=texture_block{DefaultImage:=Textures.General.T_RarityBracketRight, DefaultDesiredSize:=vector2{X:=18.0, Y:=36.0}, DefaultTint:=MakeColorFromHex("#FFFFFF44")}

                                                                                            stack_box_slot:
                                                                                                VerticalAlignment:=vertical_alignment.Center
                                                                                                HorizontalAlignment:=horizontal_alignment.Center
                                                                                                Distribution:=option{1.0}
                                                                                                Padding:=margin{Left:=0.0, Top:=0.0, Right:=0.0, Bottom:=0.0}
                                                                                                Widget:=overlay:
                                                                                                    Slots:=array:
                                                                                                        #Makeshift size box
                                                                                                        overlay_slot:
                                                                                                            VerticalAlignment:=vertical_alignment.Center
                                                                                                            HorizontalAlignment:=horizontal_alignment.Center
                                                                                                            Padding:=margin{Left:=0.0, Top:=0.0, Right:=0.0, Bottom:=0.0}
                                                                                                            Widget:=texture_block{DefaultImage:=Textures.General.T_GenSquTrans, DefaultDesiredSize:=vector2{X:=30.0, Y:=40.0}, DefaultTint:=MakeColorFromHex("#0000FF08")}

                                                                                                        #Overlay for rarity border and texture
                                                                                                        overlay_slot:
                                                                                                            VerticalAlignment:=vertical_alignment.Center
                                                                                                            HorizontalAlignment:=horizontal_alignment.Center
                                                                                                            Padding:=margin{Left:=0.0, Top:=0.0, Right:=0.0, Bottom:=0.0}
                                                                                                            Widget:=overlay:
                                                                                                                Slots:=array:
                                                                                                                    
                                                                                                                    #Rarity texture border, when the rarity slot is empty
                                                                                                                    overlay_slot:
                                                                                                                        VerticalAlignment:=vertical_alignment.Center
                                                                                                                        HorizontalAlignment:=horizontal_alignment.Center
                                                                                                                        Padding:=margin{Left:=0.0, Top:=0.0, Right:=0.0, Bottom:=0.0}
                                                                                                                        Widget:=texture_block{DefaultImage:=Textures.General.T_GenericBorderSquare2, DefaultDesiredSize:=vector2{X:=30.0, Y:=30.0}, DefaultTint:=MakeColorFromHex("#2e7be388")}

                                                                                                                    #Rarity texture
                                                                                                                    #overlay_slot:
                                                                                                                        #VerticalAlignment:=vertical_alignment.Center
                                                                                                                        #HorizontalAlignment:=horizontal_alignment.Center
                                                                                                                        #Padding:=margin{Left:=0.0, Top:=0.0, Right:=0.0, Bottom:=0.0}
                                                                                                                        #Widget:=texture_block{DefaultImage:=Textures.General.T_GenericSquare, DefaultDesiredSize:=vector2{X:=30.0, Y:=30.0}, DefaultTint:=MakeColorFromHex("#4ccc2aFF")}

                                                                                                        #Bracket
                                                                                                        #overlay_slot:
                                                                                                            #VerticalAlignment:=vertical_alignment.Fill
                                                                                                            #HorizontalAlignment:=horizontal_alignment.Fill
                                                                                                            #Padding:=margin{Left:=0.0, Top:=0.0, Right:=0.0, Bottom:=0.0}
                                                                                                            #Widget:=stack_box:
                                                                                                                #Orientation:=orientation.Horizontal
                                                                                                                #two instances of bracket texture to represent the range of rarities
                                                                                                                #Slots:=array:

                                                                                                                    #stack_box_slot:
                                                                                                                        #VerticalAlignment:=vertical_alignment.Center
                                                                                                                        #HorizontalAlignment:=horizontal_alignment.Left
                                                                                                                        #Distribution:=option{1.0}
                                                                                                                        #Padding:=margin{Left:=0.0, Top:=0.0, Right:=0.0, Bottom:=0.0}
                                                                                                                        #Widget:=texture_block{DefaultImage:=Textures.General.T_RarityBracketLeft, DefaultDesiredSize:=vector2{X:=18.0, Y:=36.0}, DefaultTint:=MakeColorFromHex("#FFFFFF44")}

                                                                                                                    #stack_box_slot:
                                                                                                                        #VerticalAlignment:=vertical_alignment.Center
                                                                                                                        #HorizontalAlignment:=horizontal_alignment.Right
                                                                                                                        #Distribution:=option{1.0}
                                                                                                                        #Padding:=margin{Left:=0.0, Top:=0.0, Right:=0.0, Bottom:=0.0}
                                                                                                                        #Widget:=texture_block{DefaultImage:=Textures.General.T_RarityBracketRight, DefaultDesiredSize:=vector2{X:=18.0, Y:=36.0}, DefaultTint:=MakeColorFromHex("#FFFFFF44")}

                                                                                            stack_box_slot:
                                                                                                VerticalAlignment:=vertical_alignment.Center
                                                                                                HorizontalAlignment:=horizontal_alignment.Center
                                                                                                Distribution:=option{1.0}
                                                                                                Padding:=margin{Left:=0.0, Top:=0.0, Right:=0.0, Bottom:=0.0}
                                                                                                Widget:=overlay:
                                                                                                    Slots:=array:
                                                                                                        #Makeshift size box
                                                                                                        overlay_slot:
                                                                                                            VerticalAlignment:=vertical_alignment.Center
                                                                                                            HorizontalAlignment:=horizontal_alignment.Center
                                                                                                            Padding:=margin{Left:=0.0, Top:=0.0, Right:=0.0, Bottom:=0.0}
                                                                                                            Widget:=texture_block{DefaultImage:=Textures.General.T_GenSquTrans, DefaultDesiredSize:=vector2{X:=30.0, Y:=40.0}, DefaultTint:=MakeColorFromHex("#0000FF08")}

                                                                                                        #Overlay for rarity border and texture
                                                                                                        overlay_slot:
                                                                                                            VerticalAlignment:=vertical_alignment.Center
                                                                                                            HorizontalAlignment:=horizontal_alignment.Center
                                                                                                            Padding:=margin{Left:=0.0, Top:=0.0, Right:=0.0, Bottom:=0.0}
                                                                                                            Widget:=overlay:
                                                                                                                Slots:=array:
                                                                                                                                                                                                                
                                                                                                                    #Rarity texture border, when the rarity slot is empty
                                                                                                                    overlay_slot:
                                                                                                                        VerticalAlignment:=vertical_alignment.Center
                                                                                                                        HorizontalAlignment:=horizontal_alignment.Center
                                                                                                                        Padding:=margin{Left:=0.0, Top:=0.0, Right:=0.0, Bottom:=0.0}
                                                                                                                        Widget:=texture_block{DefaultImage:=Textures.General.T_GenericBorderSquare2, DefaultDesiredSize:=vector2{X:=30.0, Y:=30.0}, DefaultTint:=MakeColorFromHex("#2e7be388")}
                                                                                                                                                                                                                            
                                                                                                                    #Rarity texture
                                                                                                                    #overlay_slot:
                                                                                                                        #VerticalAlignment:=vertical_alignment.Bottom
                                                                                                                        #HorizontalAlignment:=horizontal_alignment.Center
                                                                                                                        #Padding:=margin{Left:=0.0, Top:=0.0, Right:=0.0, Bottom:=0.0}
                                                                                                                        #Widget:=texture_block{DefaultImage:=Textures.General.T_GenericSquare, DefaultDesiredSize:=vector2{X:=30.0, Y:=30.0}, DefaultTint:=MakeColorFromHex("#2a54ccFF")}
                                                                                                                    
                                                                                                        #Bracket
                                                                                                        #overlay_slot:
                                                                                                            #VerticalAlignment:=vertical_alignment.Fill
                                                                                                            #HorizontalAlignment:=horizontal_alignment.Fill
                                                                                                            #Padding:=margin{Left:=0.0, Top:=0.0, Right:=0.0, Bottom:=0.0}
                                                                                                            #Widget:=stack_box:
                                                                                                                #Orientation:=orientation.Horizontal
                                                                                                                #two instances of bracket texture to represent the range of rarities
                                                                                                                #Slots:=array:

                                                                                                                    #stack_box_slot:
                                                                                                                        #VerticalAlignment:=vertical_alignment.Center
                                                                                                                        #HorizontalAlignment:=horizontal_alignment.Left
                                                                                                                        #Distribution:=option{1.0}
                                                                                                                        #Padding:=margin{Left:=0.0, Top:=0.0, Right:=0.0, Bottom:=0.0}
                                                                                                                        #Widget:=texture_block{DefaultImage:=Textures.General.T_RarityBracketLeft, DefaultDesiredSize:=vector2{X:=18.0, Y:=36.0}, DefaultTint:=MakeColorFromHex("#FFFFFF44")}

                                                                                                                    #stack_box_slot:
                                                                                                                        #VerticalAlignment:=vertical_alignment.Center
                                                                                                                        #HorizontalAlignment:=horizontal_alignment.Right
                                                                                                                        #Distribution:=option{1.0}
                                                                                                                        #Padding:=margin{Left:=0.0, Top:=0.0, Right:=0.0, Bottom:=0.0}
                                                                                                                        #Widget:=texture_block{DefaultImage:=Textures.General.T_RarityBracketRight, DefaultDesiredSize:=vector2{X:=18.0, Y:=36.0}, DefaultTint:=MakeColorFromHex("#FFFFFF44")}

                                                                                            stack_box_slot:
                                                                                                VerticalAlignment:=vertical_alignment.Center
                                                                                                HorizontalAlignment:=horizontal_alignment.Center
                                                                                                Distribution:=option{1.0}
                                                                                                Padding:=margin{Left:=0.0, Top:=0.0, Right:=0.0, Bottom:=0.0}
                                                                                                Widget:=overlay:
                                                                                                    Slots:=array:
                                                                                                        #Makeshift size box
                                                                                                        overlay_slot:
                                                                                                            VerticalAlignment:=vertical_alignment.Center
                                                                                                            HorizontalAlignment:=horizontal_alignment.Center
                                                                                                            Padding:=margin{Left:=0.0, Top:=0.0, Right:=0.0, Bottom:=0.0}
                                                                                                            Widget:=texture_block{DefaultImage:=Textures.General.T_GenSquTrans, DefaultDesiredSize:=vector2{X:=30.0, Y:=40.0}, DefaultTint:=MakeColorFromHex("#0000FF08")}

                                                                                                        #Overlay for rarity border and texture
                                                                                                        overlay_slot:
                                                                                                            VerticalAlignment:=vertical_alignment.Center
                                                                                                            HorizontalAlignment:=horizontal_alignment.Center
                                                                                                            Padding:=margin{Left:=0.0, Top:=0.0, Right:=0.0, Bottom:=0.0}
                                                                                                            Widget:=overlay:
                                                                                                                Slots:=array:

                                                                                                                    #Rarity texture border, when the rarity slot is empty
                                                                                                                    overlay_slot:
                                                                                                                        VerticalAlignment:=vertical_alignment.Center
                                                                                                                        HorizontalAlignment:=horizontal_alignment.Center
                                                                                                                        Padding:=margin{Left:=0.0, Top:=0.0, Right:=0.0, Bottom:=0.0}
                                                                                                                        Widget:=texture_block{DefaultImage:=Textures.General.T_GenericBorderSquare2, DefaultDesiredSize:=vector2{X:=30.0, Y:=30.0}, DefaultTint:=MakeColorFromHex("#2e7be388")}
                                                                                                                    
                                                                                                                    #Rarity texture
                                                                                                                    #overlay_slot:
                                                                                                                        #VerticalAlignment:=vertical_alignment.Bottom
                                                                                                                        #HorizontalAlignment:=horizontal_alignment.Center
                                                                                                                        #Padding:=margin{Left:=0.0, Top:=0.0, Right:=0.0, Bottom:=0.0}
                                                                                                                        #Widget:=texture_block{DefaultImage:=Textures.General.T_GenericSquare, DefaultDesiredSize:=vector2{X:=30.0, Y:=30.0}, DefaultTint:=MakeColorFromHex("#b32accFF")}
                                                                                                                    
                                                                                                        #Bracket
                                                                                                        #overlay_slot:
                                                                                                            #VerticalAlignment:=vertical_alignment.Fill
                                                                                                            #HorizontalAlignment:=horizontal_alignment.Fill
                                                                                                            #Padding:=margin{Left:=0.0, Top:=0.0, Right:=0.0, Bottom:=0.0}
                                                                                                            #Widget:=stack_box:
                                                                                                                #Orientation:=orientation.Horizontal
                                                                                                                #two instances of bracket texture to represent the range of rarities
                                                                                                                #Slots:=array:

                                                                                                                    #stack_box_slot:
                                                                                                                        #VerticalAlignment:=vertical_alignment.Center
                                                                                                                        #HorizontalAlignment:=horizontal_alignment.Left
                                                                                                                        #Distribution:=option{1.0}
                                                                                                                        #Padding:=margin{Left:=0.0, Top:=0.0, Right:=0.0, Bottom:=0.0}
                                                                                                                        #Widget:=texture_block{DefaultImage:=Textures.General.T_RarityBracketLeft, DefaultDesiredSize:=vector2{X:=18.0, Y:=36.0}, DefaultTint:=MakeColorFromHex("#FFFFFF44")}

                                                                                                                    #stack_box_slot:
                                                                                                                        #VerticalAlignment:=vertical_alignment.Center
                                                                                                                        #HorizontalAlignment:=horizontal_alignment.Right
                                                                                                                        #Distribution:=option{1.0}
                                                                                                                        #Padding:=margin{Left:=0.0, Top:=0.0, Right:=0.0, Bottom:=0.0}
                                                                                                                        #Widget:=texture_block{DefaultImage:=Textures.General.T_RarityBracketRight, DefaultDesiredSize:=vector2{X:=18.0, Y:=36.0}, DefaultTint:=MakeColorFromHex("#FFFFFF44")}

                                                                                            stack_box_slot:
                                                                                                VerticalAlignment:=vertical_alignment.Center
                                                                                                HorizontalAlignment:=horizontal_alignment.Center
                                                                                                Distribution:=option{1.0}
                                                                                                Padding:=margin{Left:=0.0, Top:=0.0, Right:=0.0, Bottom:=0.0}
                                                                                                Widget:=overlay:
                                                                                                    Slots:=array:
                                                                                                        #Makeshift size box
                                                                                                        overlay_slot:
                                                                                                            VerticalAlignment:=vertical_alignment.Center
                                                                                                            HorizontalAlignment:=horizontal_alignment.Center
                                                                                                            Padding:=margin{Left:=0.0, Top:=0.0, Right:=0.0, Bottom:=0.0}
                                                                                                            Widget:=texture_block{DefaultImage:=Textures.General.T_GenSquTrans, DefaultDesiredSize:=vector2{X:=30.0, Y:=40.0}, DefaultTint:=MakeColorFromHex("#0000FF08")}

                                                                                                        #Overlay for rarity border and texture
                                                                                                        overlay_slot:
                                                                                                            VerticalAlignment:=vertical_alignment.Center
                                                                                                            HorizontalAlignment:=horizontal_alignment.Center
                                                                                                            Padding:=margin{Left:=0.0, Top:=0.0, Right:=0.0, Bottom:=0.0}
                                                                                                            Widget:=overlay:
                                                                                                                Slots:=array:

                                                                                                                    #Rarity texture border, when the rarity slot is empty
                                                                                                                    overlay_slot:
                                                                                                                        VerticalAlignment:=vertical_alignment.Center
                                                                                                                        HorizontalAlignment:=horizontal_alignment.Center
                                                                                                                        Padding:=margin{Left:=0.0, Top:=0.0, Right:=0.0, Bottom:=0.0}
                                                                                                                        Widget:=texture_block{DefaultImage:=Textures.General.T_GenericBorderSquare2, DefaultDesiredSize:=vector2{X:=30.0, Y:=30.0}, DefaultTint:=MakeColorFromHex("#2e7be388")}
                                                                                                                    
                                                                                                                    #Rarity texture
                                                                                                                    #overlay_slot:
                                                                                                                        #VerticalAlignment:=vertical_alignment.Bottom
                                                                                                                        #HorizontalAlignment:=horizontal_alignment.Center
                                                                                                                        #Padding:=margin{Left:=0.0, Top:=0.0, Right:=0.0, Bottom:=0.0}
                                                                                                                        #Widget:=texture_block{DefaultImage:=Textures.General.T_GenericSquare, DefaultDesiredSize:=vector2{X:=30.0, Y:=30.0}, DefaultTint:=MakeColorFromHex("#cc8d2aFF")}

                                                                                                        #Bracket
                                                                                                        #overlay_slot:
                                                                                                            #VerticalAlignment:=vertical_alignment.Fill
                                                                                                            #HorizontalAlignment:=horizontal_alignment.Fill
                                                                                                            #Padding:=margin{Left:=0.0, Top:=0.0, Right:=0.0, Bottom:=0.0}
                                                                                                            #Widget:=stack_box:
                                                                                                                #Orientation:=orientation.Horizontal
                                                                                                                #two instances of bracket texture to represent the range of rarities
                                                                                                                #Slots:=array:

                                                                                                                    #stack_box_slot:
                                                                                                                        #VerticalAlignment:=vertical_alignment.Center
                                                                                                                        #HorizontalAlignment:=horizontal_alignment.Left
                                                                                                                        #Distribution:=option{1.0}
                                                                                                                        #Padding:=margin{Left:=0.0, Top:=0.0, Right:=0.0, Bottom:=0.0}
                                                                                                                        #Widget:=texture_block{DefaultImage:=Textures.General.T_RarityBracketLeft, DefaultDesiredSize:=vector2{X:=18.0, Y:=36.0}, DefaultTint:=MakeColorFromHex("#FFFFFF44")}

                                                                                                                    #stack_box_slot:
                                                                                                                        #VerticalAlignment:=vertical_alignment.Center
                                                                                                                        #HorizontalAlignment:=horizontal_alignment.Right
                                                                                                                        #Distribution:=option{1.0}
                                                                                                                        #Padding:=margin{Left:=0.0, Top:=0.0, Right:=0.0, Bottom:=0.0}
                                                                                                                        #Widget:=texture_block{DefaultImage:=Textures.General.T_RarityBracketRight, DefaultDesiredSize:=vector2{X:=18.0, Y:=36.0}, DefaultTint:=MakeColorFromHex("#FFFFFF44")}

                                                                                            stack_box_slot:
                                                                                                VerticalAlignment:=vertical_alignment.Center
                                                                                                HorizontalAlignment:=horizontal_alignment.Center
                                                                                                Distribution:=option{1.0}
                                                                                                Padding:=margin{Left:=0.0, Top:=0.0, Right:=0.0, Bottom:=0.0}
                                                                                                Widget:=overlay:
                                                                                                    Slots:=array:
                                                                                                        #Makeshift size box
                                                                                                        overlay_slot:
                                                                                                            VerticalAlignment:=vertical_alignment.Center
                                                                                                            HorizontalAlignment:=horizontal_alignment.Center
                                                                                                            Padding:=margin{Left:=0.0, Top:=0.0, Right:=0.0, Bottom:=0.0}
                                                                                                            Widget:=texture_block{DefaultImage:=Textures.General.T_GenSquTrans, DefaultDesiredSize:=vector2{X:=30.0, Y:=40.0}, DefaultTint:=MakeColorFromHex("#0000FF08")}

                                                                                                        #Overlay for rarity border and texture
                                                                                                        overlay_slot:
                                                                                                            VerticalAlignment:=vertical_alignment.Center
                                                                                                            HorizontalAlignment:=horizontal_alignment.Center
                                                                                                            Padding:=margin{Left:=0.0, Top:=0.0, Right:=0.0, Bottom:=0.0}
                                                                                                            Widget:=overlay:
                                                                                                                Slots:=array:
                                                                                                        
                                                                                                                    #Rarity texture border, when the rarity slot is empty
                                                                                                                    overlay_slot:
                                                                                                                        VerticalAlignment:=vertical_alignment.Center
                                                                                                                        HorizontalAlignment:=horizontal_alignment.Center
                                                                                                                        Padding:=margin{Left:=0.0, Top:=0.0, Right:=0.0, Bottom:=0.0}
                                                                                                                        Widget:=texture_block{DefaultImage:=Textures.General.T_GenericBorderSquare2, DefaultDesiredSize:=vector2{X:=30.0, Y:=30.0}, DefaultTint:=MakeColorFromHex("#2e7be388")}
                                                                                                                    
                                                                                                                    #Rarity texture
                                                                                                                    #overlay_slot:
                                                                                                                        #VerticalAlignment:=vertical_alignment.Bottom
                                                                                                                        #HorizontalAlignment:=horizontal_alignment.Center
                                                                                                                        #Padding:=margin{Left:=0.0, Top:=0.0, Right:=0.0, Bottom:=0.0}
                                                                                                                        #Widget:=texture_block{DefaultImage:=Textures.General.T_GenericSquare, DefaultDesiredSize:=vector2{X:=30.0, Y:=30.0}, DefaultTint:=MakeColorFromHex("#eacf19FF")}
                                                                                                                    
                                                                                                        #Bracket
                                                                                                        #overlay_slot:
                                                                                                            #VerticalAlignment:=vertical_alignment.Fill
                                                                                                            #HorizontalAlignment:=horizontal_alignment.Fill
                                                                                                            #Padding:=margin{Left:=0.0, Top:=0.0, Right:=0.0, Bottom:=0.0}
                                                                                                            #Widget:=stack_box:
                                                                                                                #Orientation:=orientation.Horizontal
                                                                                                                #two instances of bracket texture to represent the range of rarities
                                                                                                                #Slots:=array:

                                                                                                                    #stack_box_slot:
                                                                                                                        #VerticalAlignment:=vertical_alignment.Center
                                                                                                                        #HorizontalAlignment:=horizontal_alignment.Left
                                                                                                                        #Distribution:=option{1.0}
                                                                                                                        #Padding:=margin{Left:=0.0, Top:=0.0, Right:=0.0, Bottom:=0.0}
                                                                                                                        #Widget:=texture_block{DefaultImage:=Textures.General.T_RarityBracketLeft, DefaultDesiredSize:=vector2{X:=18.0, Y:=36.0}, DefaultTint:=MakeColorFromHex("#FFFFFF44")}

                                                                                                                    #stack_box_slot:
                                                                                                                        #VerticalAlignment:=vertical_alignment.Center
                                                                                                                        #HorizontalAlignment:=horizontal_alignment.Right
                                                                                                                        #Distribution:=option{1.0}
                                                                                                                        #Padding:=margin{Left:=0.0, Top:=0.0, Right:=0.0, Bottom:=0.0}
                                                                                                                        #Widget:=texture_block{DefaultImage:=Textures.General.T_RarityBracketRight, DefaultDesiredSize:=vector2{X:=18.0, Y:=36.0}, DefaultTint:=MakeColorFromHex("#FFFFFF44")}

                                            #Upgrade Button
                                            stack_box_slot:
                                                VerticalAlignment:=vertical_alignment.Center
                                                HorizontalAlignment:=horizontal_alignment.Center
                                                Padding:=margin{Left:=0.0, Top:=0.0, Right:=0.0, Bottom:=0.0}
                                                Widget:=button_loud{DefaultText:=StringToMessage("UPDATE\n$15")}

                    #overlay_slot:
                        #VerticalAlignment:=vertical_alignment.Fill
                        #HorizontalAlignment:=horizontal_alignment.Fill
                        #Padding:=margin{Left:=0.0, Top:=0.0, Right:=0.0, Bottom:=0.0}
                        #Widget:=texture_block{DefaultImage:=Textures.General.T_LevelPointer, DefaultDesiredSize:=vector2{X:=20.0, Y:=20.0}, DefaultTint:=MakeColorFromHex("#7de2FFFF")}
                
            LevelRow4:=overlay:
                Slots:=array:
                    overlay_slot:
                        #Makeshift size box
                        VerticalAlignment:=vertical_alignment.Center
                        HorizontalAlignment:=horizontal_alignment.Center
                        Padding:=margin{Left:=0.0, Top:=0.0, Right:=0.0, Bottom:=0.0}
                        Widget:=texture_block{DefaultImage:=Textures.General.T_GenSquTrans, DefaultDesiredSize:=vector2{X:=850.0, Y:=130.0}, DefaultTint:=MakeColorFromHex("#0000FFFF")}

                    overlay_slot:
                        #Rarity level content
                        VerticalAlignment:=vertical_alignment.Center
                        HorizontalAlignment:=horizontal_alignment.Left
                        Padding:=margin{Left:=0.0, Top:=0.0, Right:=0.0, Bottom:=0.0}
                        Widget:=overlay:
                            Slots:=array:

                                overlay_slot:
                                    #Makeshift size box
                                    VerticalAlignment:=vertical_alignment.Center
                                    HorizontalAlignment:=horizontal_alignment.Center
                                    Padding:=margin{Left:=0.0, Top:=0.0, Right:=0.0, Bottom:=0.0}
                                    Widget:=texture_block{DefaultImage:=Textures.General.T_GenericSquare, DefaultDesiredSize:=vector2{X:=850.0, Y:=130.0}, DefaultTint:=MakeColorFromHex("#010a1832")}

                                overlay_slot:
                                    #Highlight texture, over the content but under the rarity flag
                                    VerticalAlignment:=vertical_alignment.Fill
                                    HorizontalAlignment:=horizontal_alignment.Fill
                                    Padding:=margin{Left:=0.0, Top:=0.0, Right:=0.0, Bottom:=0.0}
                                    Widget:=texture_block{DefaultImage:=Textures.General.T_LevelBorder, DefaultDesiredSize:=vector2{X:=512.0, Y:=64.0}, DefaultTint:=MakeColorFromHex("#1d4b89FF")}

                                    #Horizontal stack box
                                overlay_slot:
                                    VerticalAlignment:=vertical_alignment.Fill
                                    HorizontalAlignment:=horizontal_alignment.Left
                                    Padding:=margin{Left:=0.0, Top:=0.0, Right:=0.0, Bottom:=0.0}
                                    Widget:=stack_box:
                                        Orientation:=orientation.Horizontal
                                        Slots:=array:
                                            #Level flag
                                            stack_box_slot:
                                                VerticalAlignment:=vertical_alignment.Top
                                                HorizontalAlignment:=horizontal_alignment.Left
                                                Padding:=margin{Left:=0.0, Top:=0.0, Right:=15.0, Bottom:=0.0}
                                                Widget:=overlay:

                                                    Slots:=array:
                                                        overlay_slot:
                                                            VerticalAlignment:=vertical_alignment.Center
                                                            HorizontalAlignment:=horizontal_alignment.Center
                                                            Padding:=margin{Left:=0.0, Top:=0.0, Right:=0.0, Bottom:=0.0}
                                                            #Tier level makeshift size box
                                                            Widget:=texture_block{DefaultImage:=Textures.General.T_GenSquTrans, DefaultDesiredSize:=vector2{X:=150.0, Y:=130.0}, DefaultTint:=MakeColorFromHex("#FF0000FF")}

                                                        overlay_slot:
                                                            VerticalAlignment:=vertical_alignment.Top
                                                            HorizontalAlignment:=horizontal_alignment.Left
                                                            Padding:=margin{Left:=0.0, Top:=0.0, Right:=0.0, Bottom:=0.0}
                                                            #Selected level yellow flag
                                                            Widget:=stack_box:
                                                                Orientation:=orientation.Vertical
                                                                Slots:=array:
                                                                    stack_box_slot:
                                                                        VerticalAlignment:=vertical_alignment.Center
                                                                        HorizontalAlignment:=horizontal_alignment.Center
                                                                        Padding:=margin{Left:=0.0, Top:=0.0, Right:=0.0, Bottom:=0.0}
                                                                        #level flag texture
                                                                        Widget:=overlay:
                                                                            Slots:=array:

                                                                                overlay_slot:
                                                                                    VerticalAlignment:=vertical_alignment.Center
                                                                                    HorizontalAlignment:=horizontal_alignment.Center
                                                                                    Padding:=margin{Left:=0.0, Top:=0.0, Right:=0.0, Bottom:=0.0}
                                                                                    #level flag texture
                                                                                    Widget:=texture_block{DefaultImage:=Textures.General.T_GenLevelFlag, DefaultDesiredSize:=vector2{X:=150.0, Y:=45.0}, DefaultTint:=MakeColorFromHex("#f7ff1aFF")}

                                                                                overlay_slot:
                                                                                    VerticalAlignment:=vertical_alignment.Center
                                                                                    HorizontalAlignment:=horizontal_alignment.Center
                                                                                    Padding:=margin{Left:=0.0, Top:=0.0, Right:=0.0, Bottom:=0.0}
                                                                                    #level flag text
                                                                                    Widget:=text_block{DefaultText := StringToMessage("Base LV"), DefaultTextColor:=MakeColorFromHex("#000000FF")}

                                                                    stack_box_slot:
                                                                        VerticalAlignment:=vertical_alignment.Center
                                                                        HorizontalAlignment:=horizontal_alignment.Center
                                                                        Padding:=margin{Left:=0.0, Top:=0.0, Right:=0.0, Bottom:=0.0}
                                                                        #level flag texture
                                                                        Widget:=overlay:
                                                                            Slots:=array:

                                                                                #Booster Icon - Increase Max Shield
                                                                                overlay_slot:
                                                                                    VerticalAlignment:=vertical_alignment.Center
                                                                                    HorizontalAlignment:=horizontal_alignment.Center
                                                                                    Padding:=margin{Left:=0.0, Top:=0.0, Right:=0.0, Bottom:=0.0}
                                                                                    #Weapon icon BG
                                                                                    Widget:=texture_block{DefaultImage:=Textures.Rarity.T_IconBG_Common, DefaultDesiredSize:=vector2{X:=78.0, Y:=78.0}, DefaultTint:=MakeColorFromHex("#FFFFFFFF")}
                        
                                                                                overlay_slot:
                                                                                    VerticalAlignment:=vertical_alignment.Center
                                                                                    HorizontalAlignment:=horizontal_alignment.Center
                                                                                    Padding:=margin{Left:=0.0, Top:=0.0, Right:=0.0, Bottom:=0.0}
                                                                                    #Weapon icon Border
                                                                                    Widget:=texture_block{DefaultImage:=Textures.General.T_BoosterFrame, DefaultDesiredSize:=vector2{X:=78.0, Y:=78.0}, DefaultTint:=MakeColorFromHex("#FFFFFFFF")}
                        
                                                                                overlay_slot:
                                                                                    VerticalAlignment:=vertical_alignment.Center
                                                                                    HorizontalAlignment:=horizontal_alignment.Center
                                                                                    Padding:=margin{Left:=0.0, Top:=0.0, Right:=0.0, Bottom:=0.0}
                                                                                    #Weapon icon - pistol
                                                                                    Widget:=texture_block{DefaultImage:=Textures.Boosters.T_IncreaseMaxShield, DefaultDesiredSize:=vector2{X:=70.0, Y:=70.0}, DefaultTint:=MakeColorFromHex("#FFFFFFFF")}
                        
                                                                                #overlay_slot:
                                                                                    #VerticalAlignment:=vertical_alignment.Bottom
                                                                                    #HorizontalAlignment:=horizontal_alignment.Center
                                                                                    #Padding:=margin{Left:=0.0, Top:=0.0, Right:=0.0, Bottom:=5.0}
                                                                                    ##Weapon header - pistol
                                                                                    #Widget:=texture_block{DefaultImage:=Textures.Items.T_HeaderPistol, DefaultDesiredSize:=vector2{X:=78.0, Y:=18.0}, DefaultTint:=MakeColorFromHex("#FFFFFFFF")}

                                                        #overlay_slot:
                                                            #VerticalAlignment:=vertical_alignment.Fill
                                                            #HorizontalAlignment:=horizontal_alignment.Center
                                                            #Padding:=margin{Left:=0.0, Top:=5.0, Right:=0.0, Bottom:=5.0}
                                                            #Unselected level and lock
                                                            #Widget:=overlay:
                                                                #Slots:=array:
                                                                    
                                                                    #overlay_slot:
                                                                        #VerticalAlignment:=vertical_alignment.Top
                                                                        #HorizontalAlignment:=horizontal_alignment.Center
                                                                        #Padding:=margin{Left:=0.0, Top:=0.0, Right:=0.0, Bottom:=0.0}
                                                                        #level flag texture
                                                                        #Widget:=text_block{DefaultText := StringToMessage("Base Level"), DefaultTextColor:=MakeColorFromHex("#FFFFFFFF")}

                                                                    #overlay_slot:
                                                                        #VerticalAlignment:=vertical_alignment.Bottom
                                                                        #HorizontalAlignment:=horizontal_alignment.Center
                                                                        #Padding:=margin{Left:=0.0, Top:=0.0, Right:=0.0, Bottom:=0.0}
                                                                        ##level flag texture
                                                                        #Widget:=overlay:
                                                                            #Slots:=array:
                                                                                ##Weapon Icon - Pistols
                                                                                #overlay_slot:
                                                                                    #VerticalAlignment:=vertical_alignment.Center
                                                                                    #HorizontalAlignment:=horizontal_alignment.Center
                                                                                    #Padding:=margin{Left:=0.0, Top:=0.0, Right:=0.0, Bottom:=0.0}
                                                                                    ##Weapon icon BG
                                                                                    #Widget:=texture_block{DefaultImage:=Textures.General.T_GenericSquare, DefaultDesiredSize:=vector2{X:=78.0, Y:=78.0}, DefaultTint:=MakeColorFromHex("#000c1d88")}
                        
                                                                                #overlay_slot:
                                                                                    #VerticalAlignment:=vertical_alignment.Center
                                                                                    #HorizontalAlignment:=horizontal_alignment.Center
                                                                                    #Padding:=margin{Left:=0.0, Top:=0.0, Right:=0.0, Bottom:=0.0}
                                                                                    ##Weapon icon Border
                                                                                    #Widget:=texture_block{DefaultImage:=Textures.General.T_GenericBorderSquare, DefaultDesiredSize:=vector2{X:=78.0, Y:=78.0}, DefaultTint:=MakeColorFromHex("#2e7be3FF")}
                        
                                                                                #overlay_slot:
                                                                                    #VerticalAlignment:=vertical_alignment.Top
                                                                                    #HorizontalAlignment:=horizontal_alignment.Center
                                                                                    #Padding:=margin{Left:=0.0, Top:=5.0, Right:=0.0, Bottom:=0.0}
                                                                                    ##Weapon icon - pistol
                                                                                    #Widget:=texture_block{DefaultImage:=Textures.Items.T_IconPistol, DefaultDesiredSize:=vector2{X:=50.0, Y:=50.0}, DefaultTint:=MakeColorFromHex("#FFFFFFFF")}
                        
                                                                                #overlay_slot:
                                                                                    #VerticalAlignment:=vertical_alignment.Bottom
                                                                                    #HorizontalAlignment:=horizontal_alignment.Center
                                                                                    #Padding:=margin{Left:=0.0, Top:=0.0, Right:=0.0, Bottom:=5.0}
                                                                                    ##Weapon header - pistol
                                                                                    #Widget:=texture_block{DefaultImage:=Textures.Items.T_HeaderPistol, DefaultDesiredSize:=vector2{X:=78.0, Y:=18.0}, DefaultTint:=MakeColorFromHex("#FFFFFFFF")}
                                            
                                            

                                            stack_box_slot:
                                                VerticalAlignment:=vertical_alignment.Fill
                                                HorizontalAlignment:=horizontal_alignment.Center #Center
                                                Padding:=margin{Left:=0.0, Top:=0.0, Right:=0.0, Bottom:=0.0}
                                                Widget:=overlay:
                                                    Slots:=array:
                                                        overlay_slot:
                                                            #Makeshift size box
                                                            VerticalAlignment:=vertical_alignment.Center
                                                            HorizontalAlignment:=horizontal_alignment.Center
                                                            Padding:=margin{Left:=0.0, Top:=0.0, Right:=0.0, Bottom:=0.0}
                                                            Widget:=texture_block{DefaultImage:=Textures.General.T_GenSquTrans, DefaultDesiredSize:=vector2{X:=418.0, Y:=100.0}, DefaultTint:=MakeColorFromHex("#00FF0011")}

                                                        overlay_slot:
                                                            #Vertical stack box to add all the rarity information
                                                            VerticalAlignment:=vertical_alignment.Center
                                                            HorizontalAlignment:=horizontal_alignment.Center
                                                            Padding:=margin{Left:=0.0, Top:=0.0, Right:=0.0, Bottom:=0.0}
                                                            Widget:=stack_box:
                                                                Orientation:=orientation.Vertical
                                                                Slots:=array:
                                                                    stack_box_slot:
                                                                        #Horizontal Stack box for max rarity available
                                                                        VerticalAlignment:=vertical_alignment.Top
                                                                        HorizontalAlignment:=horizontal_alignment.Center
                                                                        Padding:=margin{Left:=0.0, Top:=0.0, Right:=0.0, Bottom:=0.0}
                                                                        Widget:=stack_box:
                                                                            Orientation:=orientation.Horizontal
                                                                            Slots:=array:
                                                                                stack_box_slot:
                                                                                    VerticalAlignment:=vertical_alignment.Center
                                                                                    HorizontalAlignment:=horizontal_alignment.Center
                                                                                    Padding:=margin{Left:=0.0, Top:=0.0, Right:=0.0, Bottom:=0.0}
                                                                                    Widget:=text_block{DefaultText := StringToMessage("Increase Max Shield"), DefaultTextColor:=MakeColorFromHex("#FFFFFFFF")}

                                                                    stack_box_slot:
                                                                        #Horizontal overlay for overall rarity scheme
                                                                        VerticalAlignment:=vertical_alignment.Bottom
                                                                        HorizontalAlignment:=horizontal_alignment.Center
                                                                        Padding:=margin{Left:=0.0, Top:=0.0, Right:=0.0, Bottom:=0.0}
                                                                        Widget:=overlay:
                                                                            Slots:=array:
                                                                                
                                                                                #Makeshift size box for the rarity scheme
                                                                                overlay_slot:
                                                                                    VerticalAlignment:=vertical_alignment.Center
                                                                                    HorizontalAlignment:=horizontal_alignment.Center
                                                                                    Padding:=margin{Left:=0.0, Top:=0.0, Right:=0.0, Bottom:=0.0}
                                                                                    Widget:=texture_block{DefaultImage:=Textures.General.T_GenSquTrans, DefaultDesiredSize:=vector2{X:=255.0, Y:=50.0}, DefaultTint:=MakeColorFromHex("#FF000022")}#Red frame

                                                                                #Visible border for rarity layout
                                                                                overlay_slot:
                                                                                    VerticalAlignment:=vertical_alignment.Center
                                                                                    HorizontalAlignment:=horizontal_alignment.Center
                                                                                    Padding:=margin{Left:=0.0, Top:=0.0, Right:=0.0, Bottom:=0.0}
                                                                                    Widget:=texture_block{DefaultImage:=Textures.General.T_GenericBorderSquare3, DefaultDesiredSize:=vector2{X:=255.0, Y:=40.0}, DefaultTint:=MakeColorFromHex("#2e7be3FF")}

                                                                                #Horizontal Stack box for rarity slots
                                                                                overlay_slot:
                                                                                    VerticalAlignment:=vertical_alignment.Fill
                                                                                    HorizontalAlignment:=horizontal_alignment.Fill
                                                                                    Padding:=margin{Left:=5.0, Top:=0.0, Right:=5.0, Bottom:=0.0}
                                                                                    Widget:=stack_box:
                                                                                        Orientation:=orientation.Horizontal
                                                                                        Slots:=array:
                                                                                            #Overlay for each rarity slots, including the pointing arrow
                                                                                            stack_box_slot:
                                                                                                VerticalAlignment:=vertical_alignment.Center
                                                                                                HorizontalAlignment:=horizontal_alignment.Center
                                                                                                Distribution:=option{1.0}
                                                                                                Padding:=margin{Left:=0.0, Top:=0.0, Right:=0.0, Bottom:=0.0}
                                                                                                Widget:=overlay:
                                                                                                    Slots:=array:
                                                                                                        #Makeshift size box
                                                                                                        overlay_slot:
                                                                                                            VerticalAlignment:=vertical_alignment.Center
                                                                                                            HorizontalAlignment:=horizontal_alignment.Center
                                                                                                            Padding:=margin{Left:=0.0, Top:=0.0, Right:=0.0, Bottom:=0.0}
                                                                                                            Widget:=texture_block{DefaultImage:=Textures.General.T_GenSquTrans, DefaultDesiredSize:=vector2{X:=30.0, Y:=40.0}, DefaultTint:=MakeColorFromHex("#0000FF08")}

                                                                                                        #Overlay for rarity border and texture
                                                                                                        overlay_slot:
                                                                                                            VerticalAlignment:=vertical_alignment.Center
                                                                                                            HorizontalAlignment:=horizontal_alignment.Center
                                                                                                            Padding:=margin{Left:=0.0, Top:=0.0, Right:=0.0, Bottom:=0.0}
                                                                                                            Widget:=overlay:
                                                                                                                Slots:=array:
                                                                                                                    
                                                                                                                    #Rarity texture border, when the rarity slot is empty
                                                                                                                    #overlay_slot:
                                                                                                                        #VerticalAlignment:=vertical_alignment.Center
                                                                                                                        #HorizontalAlignment:=horizontal_alignment.Center
                                                                                                                        #Padding:=margin{Left:=0.0, Top:=0.0, Right:=0.0, Bottom:=0.0}
                                                                                                                        #Widget:=texture_block{DefaultImage:=Textures.General.T_GenericBorderSquare2, DefaultDesiredSize:=vector2{X:=30.0, Y:=30.0}, DefaultTint:=MakeColorFromHex("#2e7be388")}

                                                                                                                    #Rarity texture
                                                                                                                    overlay_slot:
                                                                                                                        VerticalAlignment:=vertical_alignment.Center
                                                                                                                        HorizontalAlignment:=horizontal_alignment.Center
                                                                                                                        Padding:=margin{Left:=0.0, Top:=0.0, Right:=0.0, Bottom:=0.0}
                                                                                                                        Widget:=texture_block{DefaultImage:=Textures.General.T_GenericSquare, DefaultDesiredSize:=vector2{X:=20.0, Y:=20.0}, DefaultTint:=MakeColorFromHex("#979797FF")}
                                                                                                        
                                                                                                                    

                                                                                                        #Bracket
                                                                                                        overlay_slot:
                                                                                                            VerticalAlignment:=vertical_alignment.Fill
                                                                                                            HorizontalAlignment:=horizontal_alignment.Fill
                                                                                                            Padding:=margin{Left:=0.0, Top:=0.0, Right:=0.0, Bottom:=0.0}
                                                                                                            Widget:=stack_box:
                                                                                                                Orientation:=orientation.Horizontal
                                                                                                                #two instances of bracket texture to represent the range of rarities
                                                                                                                Slots:=array:

                                                                                                                    stack_box_slot:
                                                                                                                        VerticalAlignment:=vertical_alignment.Center
                                                                                                                        HorizontalAlignment:=horizontal_alignment.Left
                                                                                                                        Distribution:=option{1.0}
                                                                                                                        Padding:=margin{Left:=0.0, Top:=0.0, Right:=0.0, Bottom:=0.0}
                                                                                                                        Widget:=texture_block{DefaultImage:=Textures.General.T_RarityBracketLeft, DefaultDesiredSize:=vector2{X:=18.0, Y:=36.0}, DefaultTint:=MakeColorFromHex("#FFFFFF44")}

                                                                                                                    stack_box_slot:
                                                                                                                        VerticalAlignment:=vertical_alignment.Center
                                                                                                                        HorizontalAlignment:=horizontal_alignment.Right
                                                                                                                        Distribution:=option{1.0}
                                                                                                                        Padding:=margin{Left:=0.0, Top:=0.0, Right:=0.0, Bottom:=0.0}
                                                                                                                        Widget:=texture_block{DefaultImage:=Textures.General.T_RarityBracketRight, DefaultDesiredSize:=vector2{X:=18.0, Y:=36.0}, DefaultTint:=MakeColorFromHex("#FFFFFF44")}

                                                                                            stack_box_slot:
                                                                                                VerticalAlignment:=vertical_alignment.Center
                                                                                                HorizontalAlignment:=horizontal_alignment.Center
                                                                                                Distribution:=option{1.0}
                                                                                                Padding:=margin{Left:=0.0, Top:=0.0, Right:=0.0, Bottom:=0.0}
                                                                                                Widget:=overlay:
                                                                                                    Slots:=array:
                                                                                                        #Makeshift size box
                                                                                                        overlay_slot:
                                                                                                            VerticalAlignment:=vertical_alignment.Center
                                                                                                            HorizontalAlignment:=horizontal_alignment.Center
                                                                                                            Padding:=margin{Left:=0.0, Top:=0.0, Right:=0.0, Bottom:=0.0}
                                                                                                            Widget:=texture_block{DefaultImage:=Textures.General.T_GenSquTrans, DefaultDesiredSize:=vector2{X:=30.0, Y:=40.0}, DefaultTint:=MakeColorFromHex("#0000FF08")}

                                                                                                        #Overlay for rarity border and texture
                                                                                                        overlay_slot:
                                                                                                            VerticalAlignment:=vertical_alignment.Center
                                                                                                            HorizontalAlignment:=horizontal_alignment.Center
                                                                                                            Padding:=margin{Left:=0.0, Top:=0.0, Right:=0.0, Bottom:=0.0}
                                                                                                            Widget:=overlay:
                                                                                                                Slots:=array:
                                                                                                                    
                                                                                                                    #Rarity texture border, when the rarity slot is empty
                                                                                                                    overlay_slot:
                                                                                                                        VerticalAlignment:=vertical_alignment.Center
                                                                                                                        HorizontalAlignment:=horizontal_alignment.Center
                                                                                                                        Padding:=margin{Left:=0.0, Top:=0.0, Right:=0.0, Bottom:=0.0}
                                                                                                                        Widget:=texture_block{DefaultImage:=Textures.General.T_GenericBorderSquare2, DefaultDesiredSize:=vector2{X:=30.0, Y:=30.0}, DefaultTint:=MakeColorFromHex("#2e7be388")}

                                                                                                                    #Rarity texture
                                                                                                                    #overlay_slot:
                                                                                                                        #VerticalAlignment:=vertical_alignment.Center
                                                                                                                        #HorizontalAlignment:=horizontal_alignment.Center
                                                                                                                        #Padding:=margin{Left:=0.0, Top:=0.0, Right:=0.0, Bottom:=0.0}
                                                                                                                        #Widget:=texture_block{DefaultImage:=Textures.General.T_GenericSquare, DefaultDesiredSize:=vector2{X:=30.0, Y:=30.0}, DefaultTint:=MakeColorFromHex("#4ccc2aFF")}

                                                                                                        #Bracket
                                                                                                        #overlay_slot:
                                                                                                            #VerticalAlignment:=vertical_alignment.Fill
                                                                                                            #HorizontalAlignment:=horizontal_alignment.Fill
                                                                                                            #Padding:=margin{Left:=0.0, Top:=0.0, Right:=0.0, Bottom:=0.0}
                                                                                                            #Widget:=stack_box:
                                                                                                                #Orientation:=orientation.Horizontal
                                                                                                                #two instances of bracket texture to represent the range of rarities
                                                                                                                #Slots:=array:

                                                                                                                    #stack_box_slot:
                                                                                                                        #VerticalAlignment:=vertical_alignment.Center
                                                                                                                        #HorizontalAlignment:=horizontal_alignment.Left
                                                                                                                        #Distribution:=option{1.0}
                                                                                                                        #Padding:=margin{Left:=0.0, Top:=0.0, Right:=0.0, Bottom:=0.0}
                                                                                                                        #Widget:=texture_block{DefaultImage:=Textures.General.T_RarityBracketLeft, DefaultDesiredSize:=vector2{X:=18.0, Y:=36.0}, DefaultTint:=MakeColorFromHex("#FFFFFF44")}

                                                                                                                    #stack_box_slot:
                                                                                                                        #VerticalAlignment:=vertical_alignment.Center
                                                                                                                        #HorizontalAlignment:=horizontal_alignment.Right
                                                                                                                        #Distribution:=option{1.0}
                                                                                                                        #Padding:=margin{Left:=0.0, Top:=0.0, Right:=0.0, Bottom:=0.0}
                                                                                                                        #Widget:=texture_block{DefaultImage:=Textures.General.T_RarityBracketRight, DefaultDesiredSize:=vector2{X:=18.0, Y:=36.0}, DefaultTint:=MakeColorFromHex("#FFFFFF44")}

                                                                                            stack_box_slot:
                                                                                                VerticalAlignment:=vertical_alignment.Center
                                                                                                HorizontalAlignment:=horizontal_alignment.Center
                                                                                                Distribution:=option{1.0}
                                                                                                Padding:=margin{Left:=0.0, Top:=0.0, Right:=0.0, Bottom:=0.0}
                                                                                                Widget:=overlay:
                                                                                                    Slots:=array:
                                                                                                        #Makeshift size box
                                                                                                        overlay_slot:
                                                                                                            VerticalAlignment:=vertical_alignment.Center
                                                                                                            HorizontalAlignment:=horizontal_alignment.Center
                                                                                                            Padding:=margin{Left:=0.0, Top:=0.0, Right:=0.0, Bottom:=0.0}
                                                                                                            Widget:=texture_block{DefaultImage:=Textures.General.T_GenSquTrans, DefaultDesiredSize:=vector2{X:=30.0, Y:=40.0}, DefaultTint:=MakeColorFromHex("#0000FF08")}

                                                                                                        #Overlay for rarity border and texture
                                                                                                        overlay_slot:
                                                                                                            VerticalAlignment:=vertical_alignment.Center
                                                                                                            HorizontalAlignment:=horizontal_alignment.Center
                                                                                                            Padding:=margin{Left:=0.0, Top:=0.0, Right:=0.0, Bottom:=0.0}
                                                                                                            Widget:=overlay:
                                                                                                                Slots:=array:
                                                                                                                                                                                                                
                                                                                                                    #Rarity texture border, when the rarity slot is empty
                                                                                                                    overlay_slot:
                                                                                                                        VerticalAlignment:=vertical_alignment.Center
                                                                                                                        HorizontalAlignment:=horizontal_alignment.Center
                                                                                                                        Padding:=margin{Left:=0.0, Top:=0.0, Right:=0.0, Bottom:=0.0}
                                                                                                                        Widget:=texture_block{DefaultImage:=Textures.General.T_GenericBorderSquare2, DefaultDesiredSize:=vector2{X:=30.0, Y:=30.0}, DefaultTint:=MakeColorFromHex("#2e7be388")}
                                                                                                                                                                                                                            
                                                                                                                    #Rarity texture
                                                                                                                    #overlay_slot:
                                                                                                                        #VerticalAlignment:=vertical_alignment.Bottom
                                                                                                                        #HorizontalAlignment:=horizontal_alignment.Center
                                                                                                                        #Padding:=margin{Left:=0.0, Top:=0.0, Right:=0.0, Bottom:=0.0}
                                                                                                                        #Widget:=texture_block{DefaultImage:=Textures.General.T_GenericSquare, DefaultDesiredSize:=vector2{X:=30.0, Y:=30.0}, DefaultTint:=MakeColorFromHex("#2a54ccFF")}
                                                                                                                    
                                                                                                        #Bracket
                                                                                                        #overlay_slot:
                                                                                                            #VerticalAlignment:=vertical_alignment.Fill
                                                                                                            #HorizontalAlignment:=horizontal_alignment.Fill
                                                                                                            #Padding:=margin{Left:=0.0, Top:=0.0, Right:=0.0, Bottom:=0.0}
                                                                                                            #Widget:=stack_box:
                                                                                                                #Orientation:=orientation.Horizontal
                                                                                                                #two instances of bracket texture to represent the range of rarities
                                                                                                                #Slots:=array:

                                                                                                                    #stack_box_slot:
                                                                                                                        #VerticalAlignment:=vertical_alignment.Center
                                                                                                                        #HorizontalAlignment:=horizontal_alignment.Left
                                                                                                                        #Distribution:=option{1.0}
                                                                                                                        #Padding:=margin{Left:=0.0, Top:=0.0, Right:=0.0, Bottom:=0.0}
                                                                                                                        #Widget:=texture_block{DefaultImage:=Textures.General.T_RarityBracketLeft, DefaultDesiredSize:=vector2{X:=18.0, Y:=36.0}, DefaultTint:=MakeColorFromHex("#FFFFFF44")}

                                                                                                                    #stack_box_slot:
                                                                                                                        #VerticalAlignment:=vertical_alignment.Center
                                                                                                                        #HorizontalAlignment:=horizontal_alignment.Right
                                                                                                                        #Distribution:=option{1.0}
                                                                                                                        #Padding:=margin{Left:=0.0, Top:=0.0, Right:=0.0, Bottom:=0.0}
                                                                                                                        #Widget:=texture_block{DefaultImage:=Textures.General.T_RarityBracketRight, DefaultDesiredSize:=vector2{X:=18.0, Y:=36.0}, DefaultTint:=MakeColorFromHex("#FFFFFF44")}

                                                                                            stack_box_slot:
                                                                                                VerticalAlignment:=vertical_alignment.Center
                                                                                                HorizontalAlignment:=horizontal_alignment.Center
                                                                                                Distribution:=option{1.0}
                                                                                                Padding:=margin{Left:=0.0, Top:=0.0, Right:=0.0, Bottom:=0.0}
                                                                                                Widget:=overlay:
                                                                                                    Slots:=array:
                                                                                                        #Makeshift size box
                                                                                                        overlay_slot:
                                                                                                            VerticalAlignment:=vertical_alignment.Center
                                                                                                            HorizontalAlignment:=horizontal_alignment.Center
                                                                                                            Padding:=margin{Left:=0.0, Top:=0.0, Right:=0.0, Bottom:=0.0}
                                                                                                            Widget:=texture_block{DefaultImage:=Textures.General.T_GenSquTrans, DefaultDesiredSize:=vector2{X:=30.0, Y:=40.0}, DefaultTint:=MakeColorFromHex("#0000FF08")}

                                                                                                        #Overlay for rarity border and texture
                                                                                                        overlay_slot:
                                                                                                            VerticalAlignment:=vertical_alignment.Center
                                                                                                            HorizontalAlignment:=horizontal_alignment.Center
                                                                                                            Padding:=margin{Left:=0.0, Top:=0.0, Right:=0.0, Bottom:=0.0}
                                                                                                            Widget:=overlay:
                                                                                                                Slots:=array:

                                                                                                                    #Rarity texture border, when the rarity slot is empty
                                                                                                                    overlay_slot:
                                                                                                                        VerticalAlignment:=vertical_alignment.Center
                                                                                                                        HorizontalAlignment:=horizontal_alignment.Center
                                                                                                                        Padding:=margin{Left:=0.0, Top:=0.0, Right:=0.0, Bottom:=0.0}
                                                                                                                        Widget:=texture_block{DefaultImage:=Textures.General.T_GenericBorderSquare2, DefaultDesiredSize:=vector2{X:=30.0, Y:=30.0}, DefaultTint:=MakeColorFromHex("#2e7be388")}
                                                                                                                    
                                                                                                                    #Rarity texture
                                                                                                                    #overlay_slot:
                                                                                                                        #VerticalAlignment:=vertical_alignment.Bottom
                                                                                                                        #HorizontalAlignment:=horizontal_alignment.Center
                                                                                                                        #Padding:=margin{Left:=0.0, Top:=0.0, Right:=0.0, Bottom:=0.0}
                                                                                                                        #Widget:=texture_block{DefaultImage:=Textures.General.T_GenericSquare, DefaultDesiredSize:=vector2{X:=30.0, Y:=30.0}, DefaultTint:=MakeColorFromHex("#b32accFF")}
                                                                                                                    
                                                                                                        #Bracket
                                                                                                        #overlay_slot:
                                                                                                            #VerticalAlignment:=vertical_alignment.Fill
                                                                                                            #HorizontalAlignment:=horizontal_alignment.Fill
                                                                                                            #Padding:=margin{Left:=0.0, Top:=0.0, Right:=0.0, Bottom:=0.0}
                                                                                                            #Widget:=stack_box:
                                                                                                                #Orientation:=orientation.Horizontal
                                                                                                                #two instances of bracket texture to represent the range of rarities
                                                                                                                #Slots:=array:

                                                                                                                    #stack_box_slot:
                                                                                                                        #VerticalAlignment:=vertical_alignment.Center
                                                                                                                        #HorizontalAlignment:=horizontal_alignment.Left
                                                                                                                        #Distribution:=option{1.0}
                                                                                                                        #Padding:=margin{Left:=0.0, Top:=0.0, Right:=0.0, Bottom:=0.0}
                                                                                                                        #Widget:=texture_block{DefaultImage:=Textures.General.T_RarityBracketLeft, DefaultDesiredSize:=vector2{X:=18.0, Y:=36.0}, DefaultTint:=MakeColorFromHex("#FFFFFF44")}

                                                                                                                    #stack_box_slot:
                                                                                                                        #VerticalAlignment:=vertical_alignment.Center
                                                                                                                        #HorizontalAlignment:=horizontal_alignment.Right
                                                                                                                        #Distribution:=option{1.0}
                                                                                                                        #Padding:=margin{Left:=0.0, Top:=0.0, Right:=0.0, Bottom:=0.0}
                                                                                                                        #Widget:=texture_block{DefaultImage:=Textures.General.T_RarityBracketRight, DefaultDesiredSize:=vector2{X:=18.0, Y:=36.0}, DefaultTint:=MakeColorFromHex("#FFFFFF44")}

                                                                                            stack_box_slot:
                                                                                                VerticalAlignment:=vertical_alignment.Center
                                                                                                HorizontalAlignment:=horizontal_alignment.Center
                                                                                                Distribution:=option{1.0}
                                                                                                Padding:=margin{Left:=0.0, Top:=0.0, Right:=0.0, Bottom:=0.0}
                                                                                                Widget:=overlay:
                                                                                                    Slots:=array:
                                                                                                        #Makeshift size box
                                                                                                        overlay_slot:
                                                                                                            VerticalAlignment:=vertical_alignment.Center
                                                                                                            HorizontalAlignment:=horizontal_alignment.Center
                                                                                                            Padding:=margin{Left:=0.0, Top:=0.0, Right:=0.0, Bottom:=0.0}
                                                                                                            Widget:=texture_block{DefaultImage:=Textures.General.T_GenSquTrans, DefaultDesiredSize:=vector2{X:=30.0, Y:=40.0}, DefaultTint:=MakeColorFromHex("#0000FF08")}

                                                                                                        #Overlay for rarity border and texture
                                                                                                        overlay_slot:
                                                                                                            VerticalAlignment:=vertical_alignment.Center
                                                                                                            HorizontalAlignment:=horizontal_alignment.Center
                                                                                                            Padding:=margin{Left:=0.0, Top:=0.0, Right:=0.0, Bottom:=0.0}
                                                                                                            Widget:=overlay:
                                                                                                                Slots:=array:

                                                                                                                    #Rarity texture border, when the rarity slot is empty
                                                                                                                    overlay_slot:
                                                                                                                        VerticalAlignment:=vertical_alignment.Center
                                                                                                                        HorizontalAlignment:=horizontal_alignment.Center
                                                                                                                        Padding:=margin{Left:=0.0, Top:=0.0, Right:=0.0, Bottom:=0.0}
                                                                                                                        Widget:=texture_block{DefaultImage:=Textures.General.T_GenericBorderSquare2, DefaultDesiredSize:=vector2{X:=30.0, Y:=30.0}, DefaultTint:=MakeColorFromHex("#2e7be388")}
                                                                                                                    
                                                                                                                    #Rarity texture
                                                                                                                    #overlay_slot:
                                                                                                                        #VerticalAlignment:=vertical_alignment.Bottom
                                                                                                                        #HorizontalAlignment:=horizontal_alignment.Center
                                                                                                                        #Padding:=margin{Left:=0.0, Top:=0.0, Right:=0.0, Bottom:=0.0}
                                                                                                                        #Widget:=texture_block{DefaultImage:=Textures.General.T_GenericSquare, DefaultDesiredSize:=vector2{X:=30.0, Y:=30.0}, DefaultTint:=MakeColorFromHex("#cc8d2aFF")}

                                                                                                        #Bracket
                                                                                                        #overlay_slot:
                                                                                                            #VerticalAlignment:=vertical_alignment.Fill
                                                                                                            #HorizontalAlignment:=horizontal_alignment.Fill
                                                                                                            #Padding:=margin{Left:=0.0, Top:=0.0, Right:=0.0, Bottom:=0.0}
                                                                                                            #Widget:=stack_box:
                                                                                                                #Orientation:=orientation.Horizontal
                                                                                                                #two instances of bracket texture to represent the range of rarities
                                                                                                                #Slots:=array:

                                                                                                                    #stack_box_slot:
                                                                                                                        #VerticalAlignment:=vertical_alignment.Center
                                                                                                                        #HorizontalAlignment:=horizontal_alignment.Left
                                                                                                                        #Distribution:=option{1.0}
                                                                                                                        #Padding:=margin{Left:=0.0, Top:=0.0, Right:=0.0, Bottom:=0.0}
                                                                                                                        #Widget:=texture_block{DefaultImage:=Textures.General.T_RarityBracketLeft, DefaultDesiredSize:=vector2{X:=18.0, Y:=36.0}, DefaultTint:=MakeColorFromHex("#FFFFFF44")}

                                                                                                                    #stack_box_slot:
                                                                                                                        #VerticalAlignment:=vertical_alignment.Center
                                                                                                                        #HorizontalAlignment:=horizontal_alignment.Right
                                                                                                                        #Distribution:=option{1.0}
                                                                                                                        #Padding:=margin{Left:=0.0, Top:=0.0, Right:=0.0, Bottom:=0.0}
                                                                                                                        #Widget:=texture_block{DefaultImage:=Textures.General.T_RarityBracketRight, DefaultDesiredSize:=vector2{X:=18.0, Y:=36.0}, DefaultTint:=MakeColorFromHex("#FFFFFF44")}

                                                                                            stack_box_slot:
                                                                                                VerticalAlignment:=vertical_alignment.Center
                                                                                                HorizontalAlignment:=horizontal_alignment.Center
                                                                                                Distribution:=option{1.0}
                                                                                                Padding:=margin{Left:=0.0, Top:=0.0, Right:=0.0, Bottom:=0.0}
                                                                                                Widget:=overlay:
                                                                                                    Slots:=array:
                                                                                                        #Makeshift size box
                                                                                                        overlay_slot:
                                                                                                            VerticalAlignment:=vertical_alignment.Center
                                                                                                            HorizontalAlignment:=horizontal_alignment.Center
                                                                                                            Padding:=margin{Left:=0.0, Top:=0.0, Right:=0.0, Bottom:=0.0}
                                                                                                            Widget:=texture_block{DefaultImage:=Textures.General.T_GenSquTrans, DefaultDesiredSize:=vector2{X:=30.0, Y:=40.0}, DefaultTint:=MakeColorFromHex("#0000FF08")}

                                                                                                        #Overlay for rarity border and texture
                                                                                                        overlay_slot:
                                                                                                            VerticalAlignment:=vertical_alignment.Center
                                                                                                            HorizontalAlignment:=horizontal_alignment.Center
                                                                                                            Padding:=margin{Left:=0.0, Top:=0.0, Right:=0.0, Bottom:=0.0}
                                                                                                            Widget:=overlay:
                                                                                                                Slots:=array:
                                                                                                        
                                                                                                                    #Rarity texture border, when the rarity slot is empty
                                                                                                                    overlay_slot:
                                                                                                                        VerticalAlignment:=vertical_alignment.Center
                                                                                                                        HorizontalAlignment:=horizontal_alignment.Center
                                                                                                                        Padding:=margin{Left:=0.0, Top:=0.0, Right:=0.0, Bottom:=0.0}
                                                                                                                        Widget:=texture_block{DefaultImage:=Textures.General.T_GenericBorderSquare2, DefaultDesiredSize:=vector2{X:=30.0, Y:=30.0}, DefaultTint:=MakeColorFromHex("#2e7be388")}
                                                                                                                    
                                                                                                                    #Rarity texture
                                                                                                                    #overlay_slot:
                                                                                                                        #VerticalAlignment:=vertical_alignment.Bottom
                                                                                                                        #HorizontalAlignment:=horizontal_alignment.Center
                                                                                                                        #Padding:=margin{Left:=0.0, Top:=0.0, Right:=0.0, Bottom:=0.0}
                                                                                                                        #Widget:=texture_block{DefaultImage:=Textures.General.T_GenericSquare, DefaultDesiredSize:=vector2{X:=30.0, Y:=30.0}, DefaultTint:=MakeColorFromHex("#eacf19FF")}
                                                                                                                    
                                                                                                        #Bracket
                                                                                                        #overlay_slot:
                                                                                                            #VerticalAlignment:=vertical_alignment.Fill
                                                                                                            #HorizontalAlignment:=horizontal_alignment.Fill
                                                                                                            #Padding:=margin{Left:=0.0, Top:=0.0, Right:=0.0, Bottom:=0.0}
                                                                                                            #Widget:=stack_box:
                                                                                                                #Orientation:=orientation.Horizontal
                                                                                                                #two instances of bracket texture to represent the range of rarities
                                                                                                                #Slots:=array:

                                                                                                                    #stack_box_slot:
                                                                                                                        #VerticalAlignment:=vertical_alignment.Center
                                                                                                                        #HorizontalAlignment:=horizontal_alignment.Left
                                                                                                                        #Distribution:=option{1.0}
                                                                                                                        #Padding:=margin{Left:=0.0, Top:=0.0, Right:=0.0, Bottom:=0.0}
                                                                                                                        #Widget:=texture_block{DefaultImage:=Textures.General.T_RarityBracketLeft, DefaultDesiredSize:=vector2{X:=18.0, Y:=36.0}, DefaultTint:=MakeColorFromHex("#FFFFFF44")}

                                                                                                                    #stack_box_slot:
                                                                                                                        #VerticalAlignment:=vertical_alignment.Center
                                                                                                                        #HorizontalAlignment:=horizontal_alignment.Right
                                                                                                                        #Distribution:=option{1.0}
                                                                                                                        #Padding:=margin{Left:=0.0, Top:=0.0, Right:=0.0, Bottom:=0.0}
                                                                                                                        #Widget:=texture_block{DefaultImage:=Textures.General.T_RarityBracketRight, DefaultDesiredSize:=vector2{X:=18.0, Y:=36.0}, DefaultTint:=MakeColorFromHex("#FFFFFF44")}

                                            #Upgrade Button
                                            stack_box_slot:
                                                VerticalAlignment:=vertical_alignment.Center
                                                HorizontalAlignment:=horizontal_alignment.Center
                                                Padding:=margin{Left:=0.0, Top:=0.0, Right:=0.0, Bottom:=0.0}
                                                Widget:=button_loud{DefaultText:=StringToMessage("UPDATE\n$15")}

                    #overlay_slot:
                        #VerticalAlignment:=vertical_alignment.Fill
                        #HorizontalAlignment:=horizontal_alignment.Fill
                        #Padding:=margin{Left:=0.0, Top:=0.0, Right:=0.0, Bottom:=0.0}
                        #Widget:=texture_block{DefaultImage:=Textures.General.T_LevelPointer, DefaultDesiredSize:=vector2{X:=20.0, Y:=20.0}, DefaultTint:=MakeColorFromHex("#7de2FFFF")}

            LevelRow5:=overlay:
                Slots:=array:
                    overlay_slot:
                        #Makeshift size box
                        VerticalAlignment:=vertical_alignment.Center
                        HorizontalAlignment:=horizontal_alignment.Center
                        Padding:=margin{Left:=0.0, Top:=0.0, Right:=0.0, Bottom:=0.0}
                        Widget:=texture_block{DefaultImage:=Textures.General.T_GenSquTrans, DefaultDesiredSize:=vector2{X:=850.0, Y:=130.0}, DefaultTint:=MakeColorFromHex("#0000FFFF")}

                    overlay_slot:
                        #Rarity level content
                        VerticalAlignment:=vertical_alignment.Center
                        HorizontalAlignment:=horizontal_alignment.Left
                        Padding:=margin{Left:=0.0, Top:=0.0, Right:=0.0, Bottom:=0.0}
                        Widget:=overlay:
                            Slots:=array:

                                overlay_slot:
                                    #Makeshift size box
                                    VerticalAlignment:=vertical_alignment.Center
                                    HorizontalAlignment:=horizontal_alignment.Center
                                    Padding:=margin{Left:=0.0, Top:=0.0, Right:=0.0, Bottom:=0.0}
                                    Widget:=texture_block{DefaultImage:=Textures.General.T_GenericSquare, DefaultDesiredSize:=vector2{X:=850.0, Y:=130.0}, DefaultTint:=MakeColorFromHex("#010a1832")}

                                overlay_slot:
                                    #Highlight texture, over the content but under the rarity flag
                                    VerticalAlignment:=vertical_alignment.Fill
                                    HorizontalAlignment:=horizontal_alignment.Fill
                                    Padding:=margin{Left:=0.0, Top:=0.0, Right:=0.0, Bottom:=0.0}
                                    Widget:=texture_block{DefaultImage:=Textures.General.T_LevelBorder, DefaultDesiredSize:=vector2{X:=512.0, Y:=64.0}, DefaultTint:=MakeColorFromHex("#1d4b89FF")}

                                    #Horizontal stack box
                                overlay_slot:
                                    VerticalAlignment:=vertical_alignment.Fill
                                    HorizontalAlignment:=horizontal_alignment.Left
                                    Padding:=margin{Left:=0.0, Top:=0.0, Right:=0.0, Bottom:=0.0}
                                    Widget:=stack_box:
                                        Orientation:=orientation.Horizontal
                                        Slots:=array:
                                            #Level flag
                                            stack_box_slot:
                                                VerticalAlignment:=vertical_alignment.Top
                                                HorizontalAlignment:=horizontal_alignment.Left
                                                Padding:=margin{Left:=0.0, Top:=0.0, Right:=15.0, Bottom:=0.0}
                                                Widget:=overlay:

                                                    Slots:=array:
                                                        overlay_slot:
                                                            VerticalAlignment:=vertical_alignment.Center
                                                            HorizontalAlignment:=horizontal_alignment.Center
                                                            Padding:=margin{Left:=0.0, Top:=0.0, Right:=0.0, Bottom:=0.0}
                                                            #Tier level makeshift size box
                                                            Widget:=texture_block{DefaultImage:=Textures.General.T_GenSquTrans, DefaultDesiredSize:=vector2{X:=150.0, Y:=130.0}, DefaultTint:=MakeColorFromHex("#FF0000FF")}

                                                        overlay_slot:
                                                            VerticalAlignment:=vertical_alignment.Top
                                                            HorizontalAlignment:=horizontal_alignment.Left
                                                            Padding:=margin{Left:=0.0, Top:=0.0, Right:=0.0, Bottom:=0.0}
                                                            #Selected level yellow flag
                                                            Widget:=stack_box:
                                                                Orientation:=orientation.Vertical
                                                                Slots:=array:
                                                                    stack_box_slot:
                                                                        VerticalAlignment:=vertical_alignment.Center
                                                                        HorizontalAlignment:=horizontal_alignment.Center
                                                                        Padding:=margin{Left:=0.0, Top:=0.0, Right:=0.0, Bottom:=0.0}
                                                                        #level flag texture
                                                                        Widget:=overlay:
                                                                            Slots:=array:

                                                                                overlay_slot:
                                                                                    VerticalAlignment:=vertical_alignment.Center
                                                                                    HorizontalAlignment:=horizontal_alignment.Center
                                                                                    Padding:=margin{Left:=0.0, Top:=0.0, Right:=0.0, Bottom:=0.0}
                                                                                    #level flag texture
                                                                                    Widget:=texture_block{DefaultImage:=Textures.General.T_GenLevelFlag, DefaultDesiredSize:=vector2{X:=150.0, Y:=45.0}, DefaultTint:=MakeColorFromHex("#f7ff1aFF")}

                                                                                overlay_slot:
                                                                                    VerticalAlignment:=vertical_alignment.Center
                                                                                    HorizontalAlignment:=horizontal_alignment.Center
                                                                                    Padding:=margin{Left:=0.0, Top:=0.0, Right:=0.0, Bottom:=0.0}
                                                                                    #level flag text
                                                                                    Widget:=text_block{DefaultText := StringToMessage("Base LV"), DefaultTextColor:=MakeColorFromHex("#000000FF")}

                                                                    stack_box_slot:
                                                                        VerticalAlignment:=vertical_alignment.Center
                                                                        HorizontalAlignment:=horizontal_alignment.Center
                                                                        Padding:=margin{Left:=0.0, Top:=0.0, Right:=0.0, Bottom:=0.0}
                                                                        #level flag texture
                                                                        Widget:=overlay:
                                                                            Slots:=array:

                                                                                #Booster Icon - Increase Max Shield
                                                                                overlay_slot:
                                                                                    VerticalAlignment:=vertical_alignment.Center
                                                                                    HorizontalAlignment:=horizontal_alignment.Center
                                                                                    Padding:=margin{Left:=0.0, Top:=0.0, Right:=0.0, Bottom:=0.0}
                                                                                    #Weapon icon BG
                                                                                    Widget:=texture_block{DefaultImage:=Textures.Rarity.T_IconBG_Common, DefaultDesiredSize:=vector2{X:=78.0, Y:=78.0}, DefaultTint:=MakeColorFromHex("#FFFFFFFF")}
                        
                                                                                overlay_slot:
                                                                                    VerticalAlignment:=vertical_alignment.Center
                                                                                    HorizontalAlignment:=horizontal_alignment.Center
                                                                                    Padding:=margin{Left:=0.0, Top:=0.0, Right:=0.0, Bottom:=0.0}
                                                                                    #Weapon icon Border
                                                                                    Widget:=texture_block{DefaultImage:=Textures.General.T_BoosterFrame, DefaultDesiredSize:=vector2{X:=78.0, Y:=78.0}, DefaultTint:=MakeColorFromHex("#FFFFFFFF")}
                        
                                                                                overlay_slot:
                                                                                    VerticalAlignment:=vertical_alignment.Center
                                                                                    HorizontalAlignment:=horizontal_alignment.Center
                                                                                    Padding:=margin{Left:=0.0, Top:=0.0, Right:=0.0, Bottom:=0.0}
                                                                                    #Weapon icon - pistol
                                                                                    Widget:=texture_block{DefaultImage:=Textures.Boosters.T_IncreaseMaxShield, DefaultDesiredSize:=vector2{X:=70.0, Y:=70.0}, DefaultTint:=MakeColorFromHex("#FFFFFFFF")}
                        
                                                                                #overlay_slot:
                                                                                    #VerticalAlignment:=vertical_alignment.Bottom
                                                                                    #HorizontalAlignment:=horizontal_alignment.Center
                                                                                    #Padding:=margin{Left:=0.0, Top:=0.0, Right:=0.0, Bottom:=5.0}
                                                                                    ##Weapon header - pistol
                                                                                    #Widget:=texture_block{DefaultImage:=Textures.Items.T_HeaderPistol, DefaultDesiredSize:=vector2{X:=78.0, Y:=18.0}, DefaultTint:=MakeColorFromHex("#FFFFFFFF")}

                                                        #overlay_slot:
                                                            #VerticalAlignment:=vertical_alignment.Fill
                                                            #HorizontalAlignment:=horizontal_alignment.Center
                                                            #Padding:=margin{Left:=0.0, Top:=5.0, Right:=0.0, Bottom:=5.0}
                                                            #Unselected level and lock
                                                            #Widget:=overlay:
                                                                #Slots:=array:
                                                                    
                                                                    #overlay_slot:
                                                                        #VerticalAlignment:=vertical_alignment.Top
                                                                        #HorizontalAlignment:=horizontal_alignment.Center
                                                                        #Padding:=margin{Left:=0.0, Top:=0.0, Right:=0.0, Bottom:=0.0}
                                                                        #level flag texture
                                                                        #Widget:=text_block{DefaultText := StringToMessage("Base Level"), DefaultTextColor:=MakeColorFromHex("#FFFFFFFF")}

                                                                    #overlay_slot:
                                                                        #VerticalAlignment:=vertical_alignment.Bottom
                                                                        #HorizontalAlignment:=horizontal_alignment.Center
                                                                        #Padding:=margin{Left:=0.0, Top:=0.0, Right:=0.0, Bottom:=0.0}
                                                                        ##level flag texture
                                                                        #Widget:=overlay:
                                                                            #Slots:=array:
                                                                                ##Weapon Icon - Pistols
                                                                                #overlay_slot:
                                                                                    #VerticalAlignment:=vertical_alignment.Center
                                                                                    #HorizontalAlignment:=horizontal_alignment.Center
                                                                                    #Padding:=margin{Left:=0.0, Top:=0.0, Right:=0.0, Bottom:=0.0}
                                                                                    ##Weapon icon BG
                                                                                    #Widget:=texture_block{DefaultImage:=Textures.General.T_GenericSquare, DefaultDesiredSize:=vector2{X:=78.0, Y:=78.0}, DefaultTint:=MakeColorFromHex("#000c1d88")}
                        
                                                                                #overlay_slot:
                                                                                    #VerticalAlignment:=vertical_alignment.Center
                                                                                    #HorizontalAlignment:=horizontal_alignment.Center
                                                                                    #Padding:=margin{Left:=0.0, Top:=0.0, Right:=0.0, Bottom:=0.0}
                                                                                    ##Weapon icon Border
                                                                                    #Widget:=texture_block{DefaultImage:=Textures.General.T_GenericBorderSquare, DefaultDesiredSize:=vector2{X:=78.0, Y:=78.0}, DefaultTint:=MakeColorFromHex("#2e7be3FF")}
                        
                                                                                #overlay_slot:
                                                                                    #VerticalAlignment:=vertical_alignment.Top
                                                                                    #HorizontalAlignment:=horizontal_alignment.Center
                                                                                    #Padding:=margin{Left:=0.0, Top:=5.0, Right:=0.0, Bottom:=0.0}
                                                                                    ##Weapon icon - pistol
                                                                                    #Widget:=texture_block{DefaultImage:=Textures.Items.T_IconPistol, DefaultDesiredSize:=vector2{X:=50.0, Y:=50.0}, DefaultTint:=MakeColorFromHex("#FFFFFFFF")}
                        
                                                                                #overlay_slot:
                                                                                    #VerticalAlignment:=vertical_alignment.Bottom
                                                                                    #HorizontalAlignment:=horizontal_alignment.Center
                                                                                    #Padding:=margin{Left:=0.0, Top:=0.0, Right:=0.0, Bottom:=5.0}
                                                                                    ##Weapon header - pistol
                                                                                    #Widget:=texture_block{DefaultImage:=Textures.Items.T_HeaderPistol, DefaultDesiredSize:=vector2{X:=78.0, Y:=18.0}, DefaultTint:=MakeColorFromHex("#FFFFFFFF")}
                                            
                                            

                                            stack_box_slot:
                                                VerticalAlignment:=vertical_alignment.Fill
                                                HorizontalAlignment:=horizontal_alignment.Center #Center
                                                Padding:=margin{Left:=0.0, Top:=0.0, Right:=0.0, Bottom:=0.0}
                                                Widget:=overlay:
                                                    Slots:=array:
                                                        overlay_slot:
                                                            #Makeshift size box
                                                            VerticalAlignment:=vertical_alignment.Center
                                                            HorizontalAlignment:=horizontal_alignment.Center
                                                            Padding:=margin{Left:=0.0, Top:=0.0, Right:=0.0, Bottom:=0.0}
                                                            Widget:=texture_block{DefaultImage:=Textures.General.T_GenSquTrans, DefaultDesiredSize:=vector2{X:=418.0, Y:=100.0}, DefaultTint:=MakeColorFromHex("#00FF0011")}

                                                        overlay_slot:
                                                            #Vertical stack box to add all the rarity information
                                                            VerticalAlignment:=vertical_alignment.Center
                                                            HorizontalAlignment:=horizontal_alignment.Center
                                                            Padding:=margin{Left:=0.0, Top:=0.0, Right:=0.0, Bottom:=0.0}
                                                            Widget:=stack_box:
                                                                Orientation:=orientation.Vertical
                                                                Slots:=array:
                                                                    stack_box_slot:
                                                                        #Horizontal Stack box for max rarity available
                                                                        VerticalAlignment:=vertical_alignment.Top
                                                                        HorizontalAlignment:=horizontal_alignment.Center
                                                                        Padding:=margin{Left:=0.0, Top:=0.0, Right:=0.0, Bottom:=0.0}
                                                                        Widget:=stack_box:
                                                                            Orientation:=orientation.Horizontal
                                                                            Slots:=array:
                                                                                stack_box_slot:
                                                                                    VerticalAlignment:=vertical_alignment.Center
                                                                                    HorizontalAlignment:=horizontal_alignment.Center
                                                                                    Padding:=margin{Left:=0.0, Top:=0.0, Right:=0.0, Bottom:=0.0}
                                                                                    Widget:=text_block{DefaultText := StringToMessage("Increase Max Shield"), DefaultTextColor:=MakeColorFromHex("#FFFFFFFF")}

                                                                    stack_box_slot:
                                                                        #Horizontal overlay for overall rarity scheme
                                                                        VerticalAlignment:=vertical_alignment.Bottom
                                                                        HorizontalAlignment:=horizontal_alignment.Center
                                                                        Padding:=margin{Left:=0.0, Top:=0.0, Right:=0.0, Bottom:=0.0}
                                                                        Widget:=overlay:
                                                                            Slots:=array:
                                                                                
                                                                                #Makeshift size box for the rarity scheme
                                                                                overlay_slot:
                                                                                    VerticalAlignment:=vertical_alignment.Center
                                                                                    HorizontalAlignment:=horizontal_alignment.Center
                                                                                    Padding:=margin{Left:=0.0, Top:=0.0, Right:=0.0, Bottom:=0.0}
                                                                                    Widget:=texture_block{DefaultImage:=Textures.General.T_GenSquTrans, DefaultDesiredSize:=vector2{X:=255.0, Y:=50.0}, DefaultTint:=MakeColorFromHex("#FF000022")}#Red frame

                                                                                #Visible border for rarity layout
                                                                                overlay_slot:
                                                                                    VerticalAlignment:=vertical_alignment.Center
                                                                                    HorizontalAlignment:=horizontal_alignment.Center
                                                                                    Padding:=margin{Left:=0.0, Top:=0.0, Right:=0.0, Bottom:=0.0}
                                                                                    Widget:=texture_block{DefaultImage:=Textures.General.T_GenericBorderSquare3, DefaultDesiredSize:=vector2{X:=255.0, Y:=40.0}, DefaultTint:=MakeColorFromHex("#2e7be3FF")}

                                                                                #Horizontal Stack box for rarity slots
                                                                                overlay_slot:
                                                                                    VerticalAlignment:=vertical_alignment.Fill
                                                                                    HorizontalAlignment:=horizontal_alignment.Fill
                                                                                    Padding:=margin{Left:=5.0, Top:=0.0, Right:=5.0, Bottom:=0.0}
                                                                                    Widget:=stack_box:
                                                                                        Orientation:=orientation.Horizontal
                                                                                        Slots:=array:
                                                                                            #Overlay for each rarity slots, including the pointing arrow
                                                                                            stack_box_slot:
                                                                                                VerticalAlignment:=vertical_alignment.Center
                                                                                                HorizontalAlignment:=horizontal_alignment.Center
                                                                                                Distribution:=option{1.0}
                                                                                                Padding:=margin{Left:=0.0, Top:=0.0, Right:=0.0, Bottom:=0.0}
                                                                                                Widget:=overlay:
                                                                                                    Slots:=array:
                                                                                                        #Makeshift size box
                                                                                                        overlay_slot:
                                                                                                            VerticalAlignment:=vertical_alignment.Center
                                                                                                            HorizontalAlignment:=horizontal_alignment.Center
                                                                                                            Padding:=margin{Left:=0.0, Top:=0.0, Right:=0.0, Bottom:=0.0}
                                                                                                            Widget:=texture_block{DefaultImage:=Textures.General.T_GenSquTrans, DefaultDesiredSize:=vector2{X:=30.0, Y:=40.0}, DefaultTint:=MakeColorFromHex("#0000FF08")}

                                                                                                        #Overlay for rarity border and texture
                                                                                                        overlay_slot:
                                                                                                            VerticalAlignment:=vertical_alignment.Center
                                                                                                            HorizontalAlignment:=horizontal_alignment.Center
                                                                                                            Padding:=margin{Left:=0.0, Top:=0.0, Right:=0.0, Bottom:=0.0}
                                                                                                            Widget:=overlay:
                                                                                                                Slots:=array:
                                                                                                                    
                                                                                                                    #Rarity texture border, when the rarity slot is empty
                                                                                                                    #overlay_slot:
                                                                                                                        #VerticalAlignment:=vertical_alignment.Center
                                                                                                                        #HorizontalAlignment:=horizontal_alignment.Center
                                                                                                                        #Padding:=margin{Left:=0.0, Top:=0.0, Right:=0.0, Bottom:=0.0}
                                                                                                                        #Widget:=texture_block{DefaultImage:=Textures.General.T_GenericBorderSquare2, DefaultDesiredSize:=vector2{X:=30.0, Y:=30.0}, DefaultTint:=MakeColorFromHex("#2e7be388")}

                                                                                                                    #Rarity texture
                                                                                                                    overlay_slot:
                                                                                                                        VerticalAlignment:=vertical_alignment.Center
                                                                                                                        HorizontalAlignment:=horizontal_alignment.Center
                                                                                                                        Padding:=margin{Left:=0.0, Top:=0.0, Right:=0.0, Bottom:=0.0}
                                                                                                                        Widget:=texture_block{DefaultImage:=Textures.General.T_GenericSquare, DefaultDesiredSize:=vector2{X:=20.0, Y:=20.0}, DefaultTint:=MakeColorFromHex("#979797FF")}
                                                                                                        
                                                                                                                    

                                                                                                        #Bracket
                                                                                                        overlay_slot:
                                                                                                            VerticalAlignment:=vertical_alignment.Fill
                                                                                                            HorizontalAlignment:=horizontal_alignment.Fill
                                                                                                            Padding:=margin{Left:=0.0, Top:=0.0, Right:=0.0, Bottom:=0.0}
                                                                                                            Widget:=stack_box:
                                                                                                                Orientation:=orientation.Horizontal
                                                                                                                #two instances of bracket texture to represent the range of rarities
                                                                                                                Slots:=array:

                                                                                                                    stack_box_slot:
                                                                                                                        VerticalAlignment:=vertical_alignment.Center
                                                                                                                        HorizontalAlignment:=horizontal_alignment.Left
                                                                                                                        Distribution:=option{1.0}
                                                                                                                        Padding:=margin{Left:=0.0, Top:=0.0, Right:=0.0, Bottom:=0.0}
                                                                                                                        Widget:=texture_block{DefaultImage:=Textures.General.T_RarityBracketLeft, DefaultDesiredSize:=vector2{X:=18.0, Y:=36.0}, DefaultTint:=MakeColorFromHex("#FFFFFF44")}

                                                                                                                    stack_box_slot:
                                                                                                                        VerticalAlignment:=vertical_alignment.Center
                                                                                                                        HorizontalAlignment:=horizontal_alignment.Right
                                                                                                                        Distribution:=option{1.0}
                                                                                                                        Padding:=margin{Left:=0.0, Top:=0.0, Right:=0.0, Bottom:=0.0}
                                                                                                                        Widget:=texture_block{DefaultImage:=Textures.General.T_RarityBracketRight, DefaultDesiredSize:=vector2{X:=18.0, Y:=36.0}, DefaultTint:=MakeColorFromHex("#FFFFFF44")}

                                                                                            stack_box_slot:
                                                                                                VerticalAlignment:=vertical_alignment.Center
                                                                                                HorizontalAlignment:=horizontal_alignment.Center
                                                                                                Distribution:=option{1.0}
                                                                                                Padding:=margin{Left:=0.0, Top:=0.0, Right:=0.0, Bottom:=0.0}
                                                                                                Widget:=overlay:
                                                                                                    Slots:=array:
                                                                                                        #Makeshift size box
                                                                                                        overlay_slot:
                                                                                                            VerticalAlignment:=vertical_alignment.Center
                                                                                                            HorizontalAlignment:=horizontal_alignment.Center
                                                                                                            Padding:=margin{Left:=0.0, Top:=0.0, Right:=0.0, Bottom:=0.0}
                                                                                                            Widget:=texture_block{DefaultImage:=Textures.General.T_GenSquTrans, DefaultDesiredSize:=vector2{X:=30.0, Y:=40.0}, DefaultTint:=MakeColorFromHex("#0000FF08")}

                                                                                                        #Overlay for rarity border and texture
                                                                                                        overlay_slot:
                                                                                                            VerticalAlignment:=vertical_alignment.Center
                                                                                                            HorizontalAlignment:=horizontal_alignment.Center
                                                                                                            Padding:=margin{Left:=0.0, Top:=0.0, Right:=0.0, Bottom:=0.0}
                                                                                                            Widget:=overlay:
                                                                                                                Slots:=array:
                                                                                                                    
                                                                                                                    #Rarity texture border, when the rarity slot is empty
                                                                                                                    overlay_slot:
                                                                                                                        VerticalAlignment:=vertical_alignment.Center
                                                                                                                        HorizontalAlignment:=horizontal_alignment.Center
                                                                                                                        Padding:=margin{Left:=0.0, Top:=0.0, Right:=0.0, Bottom:=0.0}
                                                                                                                        Widget:=texture_block{DefaultImage:=Textures.General.T_GenericBorderSquare2, DefaultDesiredSize:=vector2{X:=30.0, Y:=30.0}, DefaultTint:=MakeColorFromHex("#2e7be388")}

                                                                                                                    #Rarity texture
                                                                                                                    #overlay_slot:
                                                                                                                        #VerticalAlignment:=vertical_alignment.Center
                                                                                                                        #HorizontalAlignment:=horizontal_alignment.Center
                                                                                                                        #Padding:=margin{Left:=0.0, Top:=0.0, Right:=0.0, Bottom:=0.0}
                                                                                                                        #Widget:=texture_block{DefaultImage:=Textures.General.T_GenericSquare, DefaultDesiredSize:=vector2{X:=30.0, Y:=30.0}, DefaultTint:=MakeColorFromHex("#4ccc2aFF")}

                                                                                                        #Bracket
                                                                                                        #overlay_slot:
                                                                                                            #VerticalAlignment:=vertical_alignment.Fill
                                                                                                            #HorizontalAlignment:=horizontal_alignment.Fill
                                                                                                            #Padding:=margin{Left:=0.0, Top:=0.0, Right:=0.0, Bottom:=0.0}
                                                                                                            #Widget:=stack_box:
                                                                                                                #Orientation:=orientation.Horizontal
                                                                                                                #two instances of bracket texture to represent the range of rarities
                                                                                                                #Slots:=array:

                                                                                                                    #stack_box_slot:
                                                                                                                        #VerticalAlignment:=vertical_alignment.Center
                                                                                                                        #HorizontalAlignment:=horizontal_alignment.Left
                                                                                                                        #Distribution:=option{1.0}
                                                                                                                        #Padding:=margin{Left:=0.0, Top:=0.0, Right:=0.0, Bottom:=0.0}
                                                                                                                        #Widget:=texture_block{DefaultImage:=Textures.General.T_RarityBracketLeft, DefaultDesiredSize:=vector2{X:=18.0, Y:=36.0}, DefaultTint:=MakeColorFromHex("#FFFFFF44")}

                                                                                                                    #stack_box_slot:
                                                                                                                        #VerticalAlignment:=vertical_alignment.Center
                                                                                                                        #HorizontalAlignment:=horizontal_alignment.Right
                                                                                                                        #Distribution:=option{1.0}
                                                                                                                        #Padding:=margin{Left:=0.0, Top:=0.0, Right:=0.0, Bottom:=0.0}
                                                                                                                        #Widget:=texture_block{DefaultImage:=Textures.General.T_RarityBracketRight, DefaultDesiredSize:=vector2{X:=18.0, Y:=36.0}, DefaultTint:=MakeColorFromHex("#FFFFFF44")}

                                                                                            stack_box_slot:
                                                                                                VerticalAlignment:=vertical_alignment.Center
                                                                                                HorizontalAlignment:=horizontal_alignment.Center
                                                                                                Distribution:=option{1.0}
                                                                                                Padding:=margin{Left:=0.0, Top:=0.0, Right:=0.0, Bottom:=0.0}
                                                                                                Widget:=overlay:
                                                                                                    Slots:=array:
                                                                                                        #Makeshift size box
                                                                                                        overlay_slot:
                                                                                                            VerticalAlignment:=vertical_alignment.Center
                                                                                                            HorizontalAlignment:=horizontal_alignment.Center
                                                                                                            Padding:=margin{Left:=0.0, Top:=0.0, Right:=0.0, Bottom:=0.0}
                                                                                                            Widget:=texture_block{DefaultImage:=Textures.General.T_GenSquTrans, DefaultDesiredSize:=vector2{X:=30.0, Y:=40.0}, DefaultTint:=MakeColorFromHex("#0000FF08")}

                                                                                                        #Overlay for rarity border and texture
                                                                                                        overlay_slot:
                                                                                                            VerticalAlignment:=vertical_alignment.Center
                                                                                                            HorizontalAlignment:=horizontal_alignment.Center
                                                                                                            Padding:=margin{Left:=0.0, Top:=0.0, Right:=0.0, Bottom:=0.0}
                                                                                                            Widget:=overlay:
                                                                                                                Slots:=array:
                                                                                                                                                                                                                
                                                                                                                    #Rarity texture border, when the rarity slot is empty
                                                                                                                    overlay_slot:
                                                                                                                        VerticalAlignment:=vertical_alignment.Center
                                                                                                                        HorizontalAlignment:=horizontal_alignment.Center
                                                                                                                        Padding:=margin{Left:=0.0, Top:=0.0, Right:=0.0, Bottom:=0.0}
                                                                                                                        Widget:=texture_block{DefaultImage:=Textures.General.T_GenericBorderSquare2, DefaultDesiredSize:=vector2{X:=30.0, Y:=30.0}, DefaultTint:=MakeColorFromHex("#2e7be388")}
                                                                                                                                                                                                                            
                                                                                                                    #Rarity texture
                                                                                                                    #overlay_slot:
                                                                                                                        #VerticalAlignment:=vertical_alignment.Bottom
                                                                                                                        #HorizontalAlignment:=horizontal_alignment.Center
                                                                                                                        #Padding:=margin{Left:=0.0, Top:=0.0, Right:=0.0, Bottom:=0.0}
                                                                                                                        #Widget:=texture_block{DefaultImage:=Textures.General.T_GenericSquare, DefaultDesiredSize:=vector2{X:=30.0, Y:=30.0}, DefaultTint:=MakeColorFromHex("#2a54ccFF")}
                                                                                                                    
                                                                                                        #Bracket
                                                                                                        #overlay_slot:
                                                                                                            #VerticalAlignment:=vertical_alignment.Fill
                                                                                                            #HorizontalAlignment:=horizontal_alignment.Fill
                                                                                                            #Padding:=margin{Left:=0.0, Top:=0.0, Right:=0.0, Bottom:=0.0}
                                                                                                            #Widget:=stack_box:
                                                                                                                #Orientation:=orientation.Horizontal
                                                                                                                #two instances of bracket texture to represent the range of rarities
                                                                                                                #Slots:=array:

                                                                                                                    #stack_box_slot:
                                                                                                                        #VerticalAlignment:=vertical_alignment.Center
                                                                                                                        #HorizontalAlignment:=horizontal_alignment.Left
                                                                                                                        #Distribution:=option{1.0}
                                                                                                                        #Padding:=margin{Left:=0.0, Top:=0.0, Right:=0.0, Bottom:=0.0}
                                                                                                                        #Widget:=texture_block{DefaultImage:=Textures.General.T_RarityBracketLeft, DefaultDesiredSize:=vector2{X:=18.0, Y:=36.0}, DefaultTint:=MakeColorFromHex("#FFFFFF44")}

                                                                                                                    #stack_box_slot:
                                                                                                                        #VerticalAlignment:=vertical_alignment.Center
                                                                                                                        #HorizontalAlignment:=horizontal_alignment.Right
                                                                                                                        #Distribution:=option{1.0}
                                                                                                                        #Padding:=margin{Left:=0.0, Top:=0.0, Right:=0.0, Bottom:=0.0}
                                                                                                                        #Widget:=texture_block{DefaultImage:=Textures.General.T_RarityBracketRight, DefaultDesiredSize:=vector2{X:=18.0, Y:=36.0}, DefaultTint:=MakeColorFromHex("#FFFFFF44")}

                                                                                            stack_box_slot:
                                                                                                VerticalAlignment:=vertical_alignment.Center
                                                                                                HorizontalAlignment:=horizontal_alignment.Center
                                                                                                Distribution:=option{1.0}
                                                                                                Padding:=margin{Left:=0.0, Top:=0.0, Right:=0.0, Bottom:=0.0}
                                                                                                Widget:=overlay:
                                                                                                    Slots:=array:
                                                                                                        #Makeshift size box
                                                                                                        overlay_slot:
                                                                                                            VerticalAlignment:=vertical_alignment.Center
                                                                                                            HorizontalAlignment:=horizontal_alignment.Center
                                                                                                            Padding:=margin{Left:=0.0, Top:=0.0, Right:=0.0, Bottom:=0.0}
                                                                                                            Widget:=texture_block{DefaultImage:=Textures.General.T_GenSquTrans, DefaultDesiredSize:=vector2{X:=30.0, Y:=40.0}, DefaultTint:=MakeColorFromHex("#0000FF08")}

                                                                                                        #Overlay for rarity border and texture
                                                                                                        overlay_slot:
                                                                                                            VerticalAlignment:=vertical_alignment.Center
                                                                                                            HorizontalAlignment:=horizontal_alignment.Center
                                                                                                            Padding:=margin{Left:=0.0, Top:=0.0, Right:=0.0, Bottom:=0.0}
                                                                                                            Widget:=overlay:
                                                                                                                Slots:=array:

                                                                                                                    #Rarity texture border, when the rarity slot is empty
                                                                                                                    overlay_slot:
                                                                                                                        VerticalAlignment:=vertical_alignment.Center
                                                                                                                        HorizontalAlignment:=horizontal_alignment.Center
                                                                                                                        Padding:=margin{Left:=0.0, Top:=0.0, Right:=0.0, Bottom:=0.0}
                                                                                                                        Widget:=texture_block{DefaultImage:=Textures.General.T_GenericBorderSquare2, DefaultDesiredSize:=vector2{X:=30.0, Y:=30.0}, DefaultTint:=MakeColorFromHex("#2e7be388")}
                                                                                                                    
                                                                                                                    #Rarity texture
                                                                                                                    #overlay_slot:
                                                                                                                        #VerticalAlignment:=vertical_alignment.Bottom
                                                                                                                        #HorizontalAlignment:=horizontal_alignment.Center
                                                                                                                        #Padding:=margin{Left:=0.0, Top:=0.0, Right:=0.0, Bottom:=0.0}
                                                                                                                        #Widget:=texture_block{DefaultImage:=Textures.General.T_GenericSquare, DefaultDesiredSize:=vector2{X:=30.0, Y:=30.0}, DefaultTint:=MakeColorFromHex("#b32accFF")}
                                                                                                                    
                                                                                                        #Bracket
                                                                                                        #overlay_slot:
                                                                                                            #VerticalAlignment:=vertical_alignment.Fill
                                                                                                            #HorizontalAlignment:=horizontal_alignment.Fill
                                                                                                            #Padding:=margin{Left:=0.0, Top:=0.0, Right:=0.0, Bottom:=0.0}
                                                                                                            #Widget:=stack_box:
                                                                                                                #Orientation:=orientation.Horizontal
                                                                                                                #two instances of bracket texture to represent the range of rarities
                                                                                                                #Slots:=array:

                                                                                                                    #stack_box_slot:
                                                                                                                        #VerticalAlignment:=vertical_alignment.Center
                                                                                                                        #HorizontalAlignment:=horizontal_alignment.Left
                                                                                                                        #Distribution:=option{1.0}
                                                                                                                        #Padding:=margin{Left:=0.0, Top:=0.0, Right:=0.0, Bottom:=0.0}
                                                                                                                        #Widget:=texture_block{DefaultImage:=Textures.General.T_RarityBracketLeft, DefaultDesiredSize:=vector2{X:=18.0, Y:=36.0}, DefaultTint:=MakeColorFromHex("#FFFFFF44")}

                                                                                                                    #stack_box_slot:
                                                                                                                        #VerticalAlignment:=vertical_alignment.Center
                                                                                                                        #HorizontalAlignment:=horizontal_alignment.Right
                                                                                                                        #Distribution:=option{1.0}
                                                                                                                        #Padding:=margin{Left:=0.0, Top:=0.0, Right:=0.0, Bottom:=0.0}
                                                                                                                        #Widget:=texture_block{DefaultImage:=Textures.General.T_RarityBracketRight, DefaultDesiredSize:=vector2{X:=18.0, Y:=36.0}, DefaultTint:=MakeColorFromHex("#FFFFFF44")}

                                                                                            stack_box_slot:
                                                                                                VerticalAlignment:=vertical_alignment.Center
                                                                                                HorizontalAlignment:=horizontal_alignment.Center
                                                                                                Distribution:=option{1.0}
                                                                                                Padding:=margin{Left:=0.0, Top:=0.0, Right:=0.0, Bottom:=0.0}
                                                                                                Widget:=overlay:
                                                                                                    Slots:=array:
                                                                                                        #Makeshift size box
                                                                                                        overlay_slot:
                                                                                                            VerticalAlignment:=vertical_alignment.Center
                                                                                                            HorizontalAlignment:=horizontal_alignment.Center
                                                                                                            Padding:=margin{Left:=0.0, Top:=0.0, Right:=0.0, Bottom:=0.0}
                                                                                                            Widget:=texture_block{DefaultImage:=Textures.General.T_GenSquTrans, DefaultDesiredSize:=vector2{X:=30.0, Y:=40.0}, DefaultTint:=MakeColorFromHex("#0000FF08")}

                                                                                                        #Overlay for rarity border and texture
                                                                                                        overlay_slot:
                                                                                                            VerticalAlignment:=vertical_alignment.Center
                                                                                                            HorizontalAlignment:=horizontal_alignment.Center
                                                                                                            Padding:=margin{Left:=0.0, Top:=0.0, Right:=0.0, Bottom:=0.0}
                                                                                                            Widget:=overlay:
                                                                                                                Slots:=array:

                                                                                                                    #Rarity texture border, when the rarity slot is empty
                                                                                                                    overlay_slot:
                                                                                                                        VerticalAlignment:=vertical_alignment.Center
                                                                                                                        HorizontalAlignment:=horizontal_alignment.Center
                                                                                                                        Padding:=margin{Left:=0.0, Top:=0.0, Right:=0.0, Bottom:=0.0}
                                                                                                                        Widget:=texture_block{DefaultImage:=Textures.General.T_GenericBorderSquare2, DefaultDesiredSize:=vector2{X:=30.0, Y:=30.0}, DefaultTint:=MakeColorFromHex("#2e7be388")}
                                                                                                                    
                                                                                                                    #Rarity texture
                                                                                                                    #overlay_slot:
                                                                                                                        #VerticalAlignment:=vertical_alignment.Bottom
                                                                                                                        #HorizontalAlignment:=horizontal_alignment.Center
                                                                                                                        #Padding:=margin{Left:=0.0, Top:=0.0, Right:=0.0, Bottom:=0.0}
                                                                                                                        #Widget:=texture_block{DefaultImage:=Textures.General.T_GenericSquare, DefaultDesiredSize:=vector2{X:=30.0, Y:=30.0}, DefaultTint:=MakeColorFromHex("#cc8d2aFF")}

                                                                                                        #Bracket
                                                                                                        #overlay_slot:
                                                                                                            #VerticalAlignment:=vertical_alignment.Fill
                                                                                                            #HorizontalAlignment:=horizontal_alignment.Fill
                                                                                                            #Padding:=margin{Left:=0.0, Top:=0.0, Right:=0.0, Bottom:=0.0}
                                                                                                            #Widget:=stack_box:
                                                                                                                #Orientation:=orientation.Horizontal
                                                                                                                #two instances of bracket texture to represent the range of rarities
                                                                                                                #Slots:=array:

                                                                                                                    #stack_box_slot:
                                                                                                                        #VerticalAlignment:=vertical_alignment.Center
                                                                                                                        #HorizontalAlignment:=horizontal_alignment.Left
                                                                                                                        #Distribution:=option{1.0}
                                                                                                                        #Padding:=margin{Left:=0.0, Top:=0.0, Right:=0.0, Bottom:=0.0}
                                                                                                                        #Widget:=texture_block{DefaultImage:=Textures.General.T_RarityBracketLeft, DefaultDesiredSize:=vector2{X:=18.0, Y:=36.0}, DefaultTint:=MakeColorFromHex("#FFFFFF44")}

                                                                                                                    #stack_box_slot:
                                                                                                                        #VerticalAlignment:=vertical_alignment.Center
                                                                                                                        #HorizontalAlignment:=horizontal_alignment.Right
                                                                                                                        #Distribution:=option{1.0}
                                                                                                                        #Padding:=margin{Left:=0.0, Top:=0.0, Right:=0.0, Bottom:=0.0}
                                                                                                                        #Widget:=texture_block{DefaultImage:=Textures.General.T_RarityBracketRight, DefaultDesiredSize:=vector2{X:=18.0, Y:=36.0}, DefaultTint:=MakeColorFromHex("#FFFFFF44")}

                                                                                            stack_box_slot:
                                                                                                VerticalAlignment:=vertical_alignment.Center
                                                                                                HorizontalAlignment:=horizontal_alignment.Center
                                                                                                Distribution:=option{1.0}
                                                                                                Padding:=margin{Left:=0.0, Top:=0.0, Right:=0.0, Bottom:=0.0}
                                                                                                Widget:=overlay:
                                                                                                    Slots:=array:
                                                                                                        #Makeshift size box
                                                                                                        overlay_slot:
                                                                                                            VerticalAlignment:=vertical_alignment.Center
                                                                                                            HorizontalAlignment:=horizontal_alignment.Center
                                                                                                            Padding:=margin{Left:=0.0, Top:=0.0, Right:=0.0, Bottom:=0.0}
                                                                                                            Widget:=texture_block{DefaultImage:=Textures.General.T_GenSquTrans, DefaultDesiredSize:=vector2{X:=30.0, Y:=40.0}, DefaultTint:=MakeColorFromHex("#0000FF08")}

                                                                                                        #Overlay for rarity border and texture
                                                                                                        overlay_slot:
                                                                                                            VerticalAlignment:=vertical_alignment.Center
                                                                                                            HorizontalAlignment:=horizontal_alignment.Center
                                                                                                            Padding:=margin{Left:=0.0, Top:=0.0, Right:=0.0, Bottom:=0.0}
                                                                                                            Widget:=overlay:
                                                                                                                Slots:=array:
                                                                                                        
                                                                                                                    #Rarity texture border, when the rarity slot is empty
                                                                                                                    overlay_slot:
                                                                                                                        VerticalAlignment:=vertical_alignment.Center
                                                                                                                        HorizontalAlignment:=horizontal_alignment.Center
                                                                                                                        Padding:=margin{Left:=0.0, Top:=0.0, Right:=0.0, Bottom:=0.0}
                                                                                                                        Widget:=texture_block{DefaultImage:=Textures.General.T_GenericBorderSquare2, DefaultDesiredSize:=vector2{X:=30.0, Y:=30.0}, DefaultTint:=MakeColorFromHex("#2e7be388")}
                                                                                                                    
                                                                                                                    #Rarity texture
                                                                                                                    #overlay_slot:
                                                                                                                        #VerticalAlignment:=vertical_alignment.Bottom
                                                                                                                        #HorizontalAlignment:=horizontal_alignment.Center
                                                                                                                        #Padding:=margin{Left:=0.0, Top:=0.0, Right:=0.0, Bottom:=0.0}
                                                                                                                        #Widget:=texture_block{DefaultImage:=Textures.General.T_GenericSquare, DefaultDesiredSize:=vector2{X:=30.0, Y:=30.0}, DefaultTint:=MakeColorFromHex("#eacf19FF")}
                                                                                                                    
                                                                                                        #Bracket
                                                                                                        #overlay_slot:
                                                                                                            #VerticalAlignment:=vertical_alignment.Fill
                                                                                                            #HorizontalAlignment:=horizontal_alignment.Fill
                                                                                                            #Padding:=margin{Left:=0.0, Top:=0.0, Right:=0.0, Bottom:=0.0}
                                                                                                            #Widget:=stack_box:
                                                                                                                #Orientation:=orientation.Horizontal
                                                                                                                #two instances of bracket texture to represent the range of rarities
                                                                                                                #Slots:=array:

                                                                                                                    #stack_box_slot:
                                                                                                                        #VerticalAlignment:=vertical_alignment.Center
                                                                                                                        #HorizontalAlignment:=horizontal_alignment.Left
                                                                                                                        #Distribution:=option{1.0}
                                                                                                                        #Padding:=margin{Left:=0.0, Top:=0.0, Right:=0.0, Bottom:=0.0}
                                                                                                                        #Widget:=texture_block{DefaultImage:=Textures.General.T_RarityBracketLeft, DefaultDesiredSize:=vector2{X:=18.0, Y:=36.0}, DefaultTint:=MakeColorFromHex("#FFFFFF44")}

                                                                                                                    #stack_box_slot:
                                                                                                                        #VerticalAlignment:=vertical_alignment.Center
                                                                                                                        #HorizontalAlignment:=horizontal_alignment.Right
                                                                                                                        #Distribution:=option{1.0}
                                                                                                                        #Padding:=margin{Left:=0.0, Top:=0.0, Right:=0.0, Bottom:=0.0}
                                                                                                                        #Widget:=texture_block{DefaultImage:=Textures.General.T_RarityBracketRight, DefaultDesiredSize:=vector2{X:=18.0, Y:=36.0}, DefaultTint:=MakeColorFromHex("#FFFFFF44")}

                                            #Upgrade Button
                                            stack_box_slot:
                                                VerticalAlignment:=vertical_alignment.Center
                                                HorizontalAlignment:=horizontal_alignment.Center
                                                Padding:=margin{Left:=0.0, Top:=0.0, Right:=0.0, Bottom:=0.0}
                                                Widget:=button_loud{DefaultText:=StringToMessage("UPDATE\n$15")}

                    #overlay_slot:
                        #VerticalAlignment:=vertical_alignment.Fill
                        #HorizontalAlignment:=horizontal_alignment.Fill
                        #Padding:=margin{Left:=0.0, Top:=0.0, Right:=0.0, Bottom:=0.0}
                        #Widget:=texture_block{DefaultImage:=Textures.General.T_LevelPointer, DefaultDesiredSize:=vector2{X:=20.0, Y:=20.0}, DefaultTint:=MakeColorFromHex("#7de2FFFF")}
            
            #Build of the screen
            HeaderCanvas:canvas = canvas:
                Slots:=array:
                    #Defines the parameters for the canvas' slot
                    canvas_slot:
                        Anchors := anchors{Minimum := vector2{X := 0.0, Y := 0.0}, Maximum := vector2{X := 1.0, Y := 1.0}}
                        Alignment := vector2{X := 0.5, Y := 0.5}
                        Widget:=overlay:
                            Slots:=array:
                                overlay_slot:
                                    VerticalAlignment:=vertical_alignment.Fill
                                    HorizontalAlignment:=horizontal_alignment.Fill
                                    Padding:=margin{Left:=0.0, Top:=0.0, Right:=0.0, Bottom:=0.0}
                                    #Screen background
                                    Widget:=texture_block{DefaultImage:=Textures.General.T_Background, DefaultDesiredSize:=vector2{X:=1920.0, Y:=1080.0}}

                                overlay_slot:
                                    VerticalAlignment:=vertical_alignment.Center
                                    HorizontalAlignment:=horizontal_alignment.Center
                                    Padding:=margin{Left:=0.0, Top:=0.0, Right:=0.0, Bottom:=0.0}
                                    #Header
                                    Widget:=stack_box:
                                        Orientation:=orientation.Vertical
                                        Slots:=array:

                                            stack_box_slot:
                                                VerticalAlignment:=vertical_alignment.Top
                                                HorizontalAlignment:=horizontal_alignment.Center
                                                Padding:=margin{Left:=0.0, Top:=0.0, Right:=0.0, Bottom:=0.0}
                                                Widget:=HeaderStackBox

                                            stack_box_slot:
                                                
                                                VerticalAlignment:=vertical_alignment.Center
                                                HorizontalAlignment:=horizontal_alignment.Center
                                                Padding:=margin{Left:=0.0, Top:=0.0, Right:=0.0, Bottom:=0.0}
                                                Widget:=overlay:
                                                    Slots:=array:

                                                        #Makeshift size box for the lower content
                                                        overlay_slot:
                                                            VerticalAlignment:=vertical_alignment.Center
                                                            HorizontalAlignment:=horizontal_alignment.Center
                                                            Padding:=margin{Left:=0.0, Top:=0.0, Right:=0.0, Bottom:=0.0}
                                                            Widget:=texture_block{DefaultImage:=Textures.General.T_GenSquTrans, DefaultDesiredSize:=vector2{X:=1750.0, Y:=900.0}}

                                                        #Horizontal stack box for content
                                                        overlay_slot:
                                                            VerticalAlignment:=vertical_alignment.Top
                                                            HorizontalAlignment:=horizontal_alignment.Fill
                                                            Padding:=margin{Left:=0.0, Top:=0.0, Right:=0.0, Bottom:=0.0}
                                                            Widget:=stack_box:
                                                                Orientation:=orientation.Horizontal
                                                                Slots:=array:

                                                                    #List of levels and each corresponding weapon loadout
                                                                    stack_box_slot:
                                                                        VerticalAlignment:=vertical_alignment.Center
                                                                        HorizontalAlignment:=horizontal_alignment.Left
                                                                        Distribution:=option{1.0}
                                                                        Padding:=margin{Left:=0.0, Top:=0.0, Right:=0.0, Bottom:=0.0}
                                                                        Widget:= stack_box:
                                                                            Orientation:=orientation.Vertical
                                                                            Slots:=array:
                                                                                
                                                                                stack_box_slot:
                                                                                    VerticalAlignment:=vertical_alignment.Center
                                                                                    HorizontalAlignment:=horizontal_alignment.Center
                                                                                    Padding:=margin{Left:=0.0, Top:=8.0, Right:=0.0, Bottom:=0.0}
                                                                                    Widget:=LevelRow0

                                                                                stack_box_slot:
                                                                                    VerticalAlignment:=vertical_alignment.Center
                                                                                    HorizontalAlignment:=horizontal_alignment.Center
                                                                                    Padding:=margin{Left:=0.0, Top:=8.0, Right:=0.0, Bottom:=0.0}
                                                                                    Widget:=LevelRow1

                                                                                stack_box_slot:
                                                                                    VerticalAlignment:=vertical_alignment.Center
                                                                                    HorizontalAlignment:=horizontal_alignment.Center
                                                                                    Padding:=margin{Left:=0.0, Top:=8.0, Right:=0.0, Bottom:=0.0}
                                                                                    Widget:=LevelRow2

                                                                                stack_box_slot:
                                                                                    VerticalAlignment:=vertical_alignment.Center
                                                                                    HorizontalAlignment:=horizontal_alignment.Center
                                                                                    Padding:=margin{Left:=0.0, Top:=8.0, Right:=0.0, Bottom:=0.0}
                                                                                    Widget:=LevelRow3

                                                                                stack_box_slot:
                                                                                    VerticalAlignment:=vertical_alignment.Center
                                                                                    HorizontalAlignment:=horizontal_alignment.Center
                                                                                    Padding:=margin{Left:=0.0, Top:=8.0, Right:=0.0, Bottom:=0.0}
                                                                                    Widget:=LevelRow4

                                                                                stack_box_slot:
                                                                                    VerticalAlignment:=vertical_alignment.Center
                                                                                    HorizontalAlignment:=horizontal_alignment.Center
                                                                                    Padding:=margin{Left:=0.0, Top:=8.0, Right:=0.0, Bottom:=0.0}
                                                                                    Widget:=LevelRow5


                                                                    #Weapon loadout toggle, level update
                                                                    stack_box_slot:
                                                                        VerticalAlignment:=vertical_alignment.Center
                                                                        HorizontalAlignment:=horizontal_alignment.Right
                                                                        #Distribution:=option{1.0}
                                                                        Padding:=margin{Left:=0.0, Top:=0.0, Right:=0.0, Bottom:=0.0}
                                                                        Widget:=overlay:
                                                                            Slots:=array:

                                                                                overlay_slot:
                                                                                    VerticalAlignment:=vertical_alignment.Center
                                                                                    HorizontalAlignment:=horizontal_alignment.Center
                                                                                    #Distribution:=option{1.0}
                                                                                    Padding:=margin{Left:=0.0, Top:=0.0, Right:=0.0, Bottom:=0.0}
                                                                                    Widget:=texture_block{DefaultImage:=Textures.General.T_GenSquTrans, DefaultDesiredSize:=vector2{X:=275.0, Y:=700.0}, DefaultTint:=MakeColorFromHex("#FF0000FF")}#LoadoutCard
                                                                            
                                                                                overlay_slot:
                                                                                    VerticalAlignment:=vertical_alignment.Bottom
                                                                                    HorizontalAlignment:=horizontal_alignment.Right
                                                                                    #Distribution:=option{1.0}
                                                                                    Padding:=margin{Left:=0.0, Top:=0.0, Right:=0.0, Bottom:=0.0}
                                                                                    Widget:=stack_box:
                                                                                        Orientation:=orientation.Vertical
                                                                                        Slots:=array:

                                                                                            
                                                                                
            if(PlayerUI:=GetPlayerUI[Player]):
                PlayerUI.AddWidget(HeaderCanvas)