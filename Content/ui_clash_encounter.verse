
using { /Fortnite.com/Devices }
using { /Fortnite.com/UI }
using { /Verse.org/Simulation }
using { /UnrealEngine.com/Temporary/Diagnostics }
using { /UnrealEngine.com/Temporary/SpatialMath }
using { /UnrealEngine.com/Temporary/UI }
using { /Verse.org/Colors/NamedColors }

ui_clash_encounter := class(clash_ui):

    RedTeamSlider : slider_regular = slider_regular{}

    BlueTeamSlider : slider_regular = slider_regular{}
    
    var RedTeamBar : texture_block = texture_block{DefaultImage:=Textures.Clash.T_hud_middle}
    
    var BlueTeamBar : texture_block = texture_block{DefaultImage:=Textures.Clash.T_hud_middle}

    var RedInvisibleBar : color_block = color_block{DefaultColor := White, DefaultOpacity := 0.0, DefaultDesiredSize:=vector2{X := 0.0, Y := 25.6}}

    var BlueInvisibleBar : color_block = color_block{DefaultColor := White, DefaultOpacity := 0.0, DefaultDesiredSize:=vector2{X := 0.0, Y := 25.6}}

    var RedTeamLimiter : texture_block = texture_block{DefaultImage:=Textures.Clash.T_hud_middle}
    
    var BlueTeamLimiter : texture_block = texture_block{DefaultImage:=Textures.Clash.T_hud_middle}

    var RedLimiterStackBox : stack_box = stack_box{Orientation := orientation.Horizontal}
    
    var BlueLimiterStackBox : stack_box = stack_box{Orientation := orientation.Horizontal}
    
    var SubscribedSpawners : []guard_spawner_device = array{}

    var CancellableArray : []cancelable = array{}

    var CurrentRoom : string = ""

    #------------------------------------------------------------------------------------------------------------
    CreateUI<override>():canvas_slot =
        set UICanvas = canvas_slot:
            Anchors := anchors{Minimum := vector2{X := 0.5, Y := 0.0}, Maximum := vector2{X := 0.5, Y := 0.0}}
            Alignment := vector2{X := 0.5, Y := 0.0}
            Offsets:=margin{Left:=0.0, Top:=150.0, Right:=0.0, Bottom:=0.0}
            Widget:=stack_box:
                Orientation:=orientation.Vertical
                Slots:=array:
                    stack_box_slot:
                        Padding:=margin{Bottom:=-25.0}
                        Widget:=text_block:
                            DefaultText := StringToMessage("Room Clash!")
                            DefaultTextColor := White
                            DefaultJustification := text_justification.Center
                            DefaultOverflowPolicy := text_overflow_policy.Clip
                    stack_box_slot:
                        Widget:=stack_box:
                            Orientation:=orientation.Horizontal
                            Slots:=array:
                                stack_box_slot:
                                    VerticalAlignment:=vertical_alignment.Fill
                                    HorizontalAlignment:=horizontal_alignment.Fill
                                    Padding:=margin{Right:=-30.0}
                                    Widget:=overlay:
                                        Slots:=array:
                                            overlay_slot:
                                                Widget:=texture_block:
                                                    DefaultDesiredSize:=vector2{X := 128.0, Y := 128.0}
                                                    DefaultImage:=Textures.Clash.T_star_red
                                            overlay_slot:
                                                Widget:=texture_block:
                                                    DefaultDesiredSize:=vector2{X := 42.0, Y := 42.0}
                                                    DefaultImage:=Textures.Clash.T_elimination
                                stack_box_slot:
                                    VerticalAlignment:=vertical_alignment.Center
                                    HorizontalAlignment:=horizontal_alignment.Center
                                    Widget:=ContructProgressBar()
                                stack_box_slot:
                                    VerticalAlignment:=vertical_alignment.Fill
                                    HorizontalAlignment:=horizontal_alignment.Fill
                                    Padding:=margin{Left:=-30.0}
                                    Widget:=overlay:
                                        Slots:=array:
                                            overlay_slot:
                                                Widget:=texture_block:
                                                    DefaultDesiredSize:=vector2{X := 128.0, Y := 128.0}
                                                    DefaultImage:=Textures.Clash.T_star_blue
                                            overlay_slot:
                                                Widget:=texture_block:
                                                    DefaultDesiredSize:=vector2{X := 42.0, Y := 42.0}
                                                    DefaultImage:=Textures.Clash.T_elimination
        return UICanvas
        
    #----------------------------------------------------------------------------------------------------------------------
    ContructProgressBar() : overlay =
        ProgressOverlay := overlay :
            Slots:=array:
                overlay_slot:
                    VerticalAlignment:=vertical_alignment.Center
                    HorizontalAlignment:=horizontal_alignment.Center
                    Widget:=texture_block:
                        DefaultDesiredSize:=vector2{X := 256.0, Y := 16.0}
                        DefaultImage:=Textures.Clash.T_hud_base
                overlay_slot:
                    VerticalAlignment:=vertical_alignment.Center
                    HorizontalAlignment:=horizontal_alignment.Left
                    Padding:=margin{Left:=5.0}
                    Widget:=ConstructRedBar()
                overlay_slot:
                    VerticalAlignment:=vertical_alignment.Center
                    HorizontalAlignment:=horizontal_alignment.Right
                    Padding:=margin{Right:=5.0}
                    Widget:=ConstructBlueBar()
                overlay_slot:
                    VerticalAlignment:=vertical_alignment.Center
                    HorizontalAlignment:=horizontal_alignment.Left
                    Widget:=ConstructRedLimiterStackBox()
                overlay_slot:
                    VerticalAlignment:=vertical_alignment.Center
                    HorizontalAlignment:=horizontal_alignment.Right
                    Widget:=ConstructBlueLimiterStackBox()
                overlay_slot:
                    VerticalAlignment:=vertical_alignment.Center
                    HorizontalAlignment:=horizontal_alignment.Center
                    Widget:=texture_block:
                        DefaultDesiredSize:=vector2{X := 16.0, Y := 16.0}
                        DefaultImage:=Textures.Clash.T_hud_middle
                overlay_slot:
                    VerticalAlignment:=vertical_alignment.Center
                    HorizontalAlignment:=horizontal_alignment.Center
                    Widget:=texture_block:
                        DefaultDesiredSize:=vector2{X := 256.0, Y := 16.0}
                        DefaultImage:=Textures.Clash.T_hud_base_transparent

        return ProgressOverlay

    #-----------------------------------------------------------------------------------------------------------------------
    ConstructBlueBar() : texture_block =
        set BlueTeamBar = texture_block:
            DefaultDesiredSize:=vector2{X := 128.0, Y := 12.8}
            DefaultImage:=Textures.Clash.T_hud_blue
        return BlueTeamBar

    #-----------------------------------------------------------------------------------------------------------------------
    ConstructRedBar() : texture_block =
        set RedTeamBar = texture_block:
            DefaultDesiredSize:=vector2{X := 128.0, Y := 12.8}
            DefaultImage:=Textures.Clash.T_hud_red
        return RedTeamBar

    #-----------------------------------------------------------------------------------------------------------------------
    ConstructBlueLimiterStackBox() : stack_box =
        set BlueLimiterStackBox = stack_box:
            Orientation:=orientation.Horizontal
            Slots:=array:
                stack_box_slot:
                    Widget:=ConstructBlueLimiter()
                stack_box_slot:
                    Widget:=BlueInvisibleBar
        return BlueLimiterStackBox

    #-----------------------------------------------------------------------------------------------------------------------
    ConstructBlueLimiter() : texture_block =
        set BlueTeamLimiter = texture_block:
            DefaultDesiredSize:=vector2{X := 6.0, Y := 12.8}
            DefaultImage:=Textures.Clash.T_hud_blue_detail
        return BlueTeamLimiter

    #-----------------------------------------------------------------------------------------------------------------------
    ConstructRedLimiterStackBox() : stack_box =
        set RedLimiterStackBox = stack_box:
            Orientation:=orientation.Horizontal
            Slots:=array:
                stack_box_slot:
                    Widget:=RedInvisibleBar
                stack_box_slot:
                    Widget:=ConstructRedLimiter()
        return RedLimiterStackBox

    #-----------------------------------------------------------------------------------------------------------------------
    ConstructRedLimiter() : texture_block =
        set RedTeamLimiter = texture_block:
            DefaultDesiredSize:=vector2{X := 6.0, Y := 12.8}
            DefaultImage:=Textures.Clash.T_hud_red_detail
        return RedTeamLimiter

    #------------------------------------------------------------------------------------------------------------
    UpdateSliders(AmountToShow : int, RoomName : string, RoomTeamOwner : team_id, BlueKills : int, RedKills : int) : void =
        if (RoomName = "Red Lobby" or RoomName = "Blue Lobby"):
            return
        if (RoomTeamOwner <> team_id.Neutral):
            return
        set CurrentRoom = RoomName
        
        RedTeamLimiter.SetDesiredSize(vector2{X := 0.0, Y := 12.8})
        BlueTeamLimiter.SetDesiredSize(vector2{X := 0.0, Y := 12.8})

        RedTeamBar.SetDesiredSize(vector2{X := 0.0, Y := 12.8})
        BlueTeamBar.SetDesiredSize(vector2{X := 0.0, Y := 12.8})

        UpdateBar(RedKills, BlueKills, AmountToShow)

    #-----------------------------------------------------------------------------------------------------------------------
    UpdateKills(RedTeamKills : int, BlueTeamKills : int, TotalEnemies : int) : void =
        UpdateBar(RedTeamKills, BlueTeamKills, TotalEnemies)

    #-----------------------------------------------------------------------------------------------------------------------
    UpdateBar(RedTeamKills : int, BlueTeamKills : int, TotalEnemies : int) : void =
        FloatAmount : float = 1.0 * TotalEnemies
        
        var RedBarValue : float = 0.0
        var BlueBarValue : float = 0.0

        set RedBarValue = (250.0 / FloatAmount) * RedTeamKills
        set BlueBarValue = (250.0 / FloatAmount) * BlueTeamKills

        RedTeamBar.SetDesiredSize(vector2{X := RedBarValue, Y := 12.8})
        BlueTeamBar.SetDesiredSize(vector2{X := BlueBarValue, Y := 12.8})

        if (RedTeamKills > 0):
            RedTeamLimiter.SetDesiredSize(vector2{X := 10.0, Y := 12.8})
            RedInvisibleBar.SetDesiredSize(vector2{X := RedBarValue - 4.0, Y := 12.8})

        if (BlueTeamKills > 0):
            BlueTeamLimiter.SetDesiredSize(vector2{X := 10.0, Y := 12.8})
            BlueInvisibleBar.SetDesiredSize(vector2{X := BlueBarValue - 4.0, Y := 12.8})

        if(TotalEnemies = (BlueTeamKills + RedTeamKills) and TotalEnemies > 0):
            SetVisibility(false)