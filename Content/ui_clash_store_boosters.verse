using { /Fortnite.com/Devices }
using { /Fortnite.com/UI}
using { /Verse.org/Simulation }
using { /Verse.org/Colors }
using { /Verse.org/Colors/NamedColors }
using { /UnrealEngine.com/Temporary/UI }
using { /UnrealEngine.com/Temporary/SpatialMath }

ui_clash_store_boosters := class(creative_device):

    OnBegin<override>()<suspends>:void=
        CreateCanvas()
    
    CreateCanvas():void=
        Players:=GetPlayspace().GetPlayers()
        if(Player:=Players[0]):
            #Structure for store header
            BaseHeader:=overlay:
                Slots:=array:

                    #Makeshift size box
                    overlay_slot:
                        VerticalAlignment:=vertical_alignment.Center
                        HorizontalAlignment:=horizontal_alignment.Center
                        Padding:=margin{Left:=0.0, Top:=0.0, Right:=0.0, Bottom:=0.0}
                        Widget:=texture_block{DefaultImage:=Textures.General.T_GenSquTrans, DefaultDesiredSize:=vector2{X:=1800.0, Y:=100.0}, DefaultTint:=MakeColorFromHex("#FF0000FF")}

                    #Header Layout
                    overlay_slot:
                        VerticalAlignment:=vertical_alignment.Center
                        HorizontalAlignment:=horizontal_alignment.Fill
                        Padding:=margin{Left:=0.0, Top:=0.0, Right:=0.0, Bottom:=0.0}
                        Widget:=stack_box:
                            Orientation:=orientation.Horizontal
                            Slots:=array:
                                
                                #Coin$ counter
                                stack_box_slot:
                                    VerticalAlignment:=vertical_alignment.Center
                                    HorizontalAlignment:=horizontal_alignment.Left
                                    Distribution:=option{0.33}
                                    Padding:=margin{Left:=30.0, Top:=0.0, Right:=0.0, Bottom:=0.0}
                                    Widget:=overlay:
                                        Slots:=array:

                                            #Makeshift size box
                                            #overlay_slot:
                                                #VerticalAlignment:=vertical_alignment.Center
                                                #HorizontalAlignment:=horizontal_alignment.Center
                                                ##Distribution:=option{0.33}
                                                #Padding:=margin{Left:=0.0, Top:=0.0, Right:=0.0, Bottom:=0.0}
                                                #Widget:=texture_block{DefaultImage:=Textures.General.T_GenericSquare, DefaultDesiredSize:=vector2{X:=650.0, Y:=50.0}, DefaultTint:=MakeColorFromHex("#FFFFFFFF")}

                                            #Coin$ counter
                                            overlay_slot:
                                                VerticalAlignment:=vertical_alignment.Center
                                                HorizontalAlignment:=horizontal_alignment.Left
                                                #Distribution:=option{0.33}
                                                Padding:=margin{Left:=0.0, Top:=0.0, Right:=0.0, Bottom:=0.0}
                                                Widget:=texture_block{DefaultImage:=Textures.Currency.T_GoldSlot, DefaultDesiredSize:=vector2{X:=180.0, Y:=50.0}, DefaultTint:=MakeColorFromHex("#FFFFFFFF")}

                                            #Coin$ counter layout
                                            overlay_slot:
                                                VerticalAlignment:=vertical_alignment.Center
                                                HorizontalAlignment:=horizontal_alignment.Left
                                                #Distribution:=option{1.0}
                                                Padding:=margin{Left:=0.0, Top:=0.0, Right:=0.0, Bottom:=0.0}
                                                Widget:=stack_box:
                                                    Orientation:=orientation.Horizontal
                                                    Slots:=array:

                                                        #Coin$ icon
                                                        stack_box_slot:
                                                            VerticalAlignment:=vertical_alignment.Center
                                                            HorizontalAlignment:=horizontal_alignment.Left
                                                            Distribution:=option{1.0}
                                                            Padding:=margin{Left:=2.0, Top:=0.0, Right:=0.0, Bottom:=0.0}
                                                            Widget:=texture_block{DefaultImage:=Textures.Currency.T_Gold, DefaultDesiredSize:=vector2{X:=40.0, Y:=40.0}, DefaultTint:=MakeColorFromHex("#FFFFFFFF")}

                                                        #Coin$ value
                                                        stack_box_slot:
                                                            VerticalAlignment:=vertical_alignment.Center
                                                            HorizontalAlignment:=horizontal_alignment.Left
                                                            Distribution:=option{1.0}
                                                            Padding:=margin{Left:=5.0, Top:=0.0, Right:=0.0, Bottom:=0.0}
                                                            Widget:=text_block{DefaultText := StringToMessage("450"), DefaultTextColor:=MakeColorFromHex("#000000FF"), DefaultShadowColor:=MakeColorFromHex("#000000FF"), DefaultShadowOffset:=option{vector2{X:=3.0, Y:=3.0}}}

                                #Header
                                stack_box_slot:
                                    VerticalAlignment:=vertical_alignment.Center
                                    HorizontalAlignment:=horizontal_alignment.Center
                                    Distribution:=option{0.33}
                                    Padding:=margin{Left:=0.0, Top:=15.0, Right:=0.0, Bottom:=0.0}
                                    Widget:=text_block{DefaultText := StringToMessage("BOOSTER Rewards"), DefaultTextColor:=MakeColorFromHex("#FFFFFFFF"), DefaultShadowColor:=MakeColorFromHex("#000000FF"), DefaultShadowOffset:=option{vector2{X:=3.0, Y:=3.0}}}

                                #Close Button
                                stack_box_slot:
                                    VerticalAlignment:=vertical_alignment.Center
                                    HorizontalAlignment:=horizontal_alignment.Right
                                    Distribution:=option{0.33}
                                    Padding:=margin{Left:=0.0, Top:=0.0, Right:=30.0, Bottom:=0.0}
                                    Widget:=button_loud{DefaultText:=StringToMessage("X")}



            #Structure for pager
            BoosterPager:=overlay:
                Slots:=array:

                    #Makeshift Size Box
                    overlay_slot:
                        VerticalAlignment:=vertical_alignment.Center
                        HorizontalAlignment:=horizontal_alignment.Center
                        #Distribution:=option{0.33}
                        Padding:=margin{Left:=0.0, Top:=0.0, Right:=0.0, Bottom:=0.0}
                        Widget:=texture_block{DefaultImage:=Textures.General.T_GenSquTrans, DefaultDesiredSize:=vector2{X:=1800.0, Y:=160.0}, DefaultTint:=MakeColorFromHex("#00FF00FF")}

                    overlay_slot:
                        VerticalAlignment:=vertical_alignment.Top
                        HorizontalAlignment:=horizontal_alignment.Center
                        #Distribution:=option{0.33}
                        Padding:=margin{Left:=0.0, Top:=10.0, Right:=0.0, Bottom:=0.0}
                        Widget:=stack_box:
                            Orientation:=orientation.Vertical
                            Slots:=array:

                                #Booster description
                                stack_box_slot:
                                    VerticalAlignment:=vertical_alignment.Center
                                    HorizontalAlignment:=horizontal_alignment.Center
                                    #Distribution:=option{0.33}
                                    Padding:=margin{Left:=0.0, Top:=0.0, Right:=0.0, Bottom:=0.0}
                                    Widget:=text_block{DefaultText := StringToMessage("Boosters are temporary buffs that can be earned when clearing a room"), DefaultTextColor:=MakeColorFromHex("#FFFFFFFF"), DefaultShadowColor:=MakeColorFromHex("#000000FF"), DefaultShadowOffset:=option{vector2{X:=3.0, Y:=3.0}}}

                                #Stroke texture
                                stack_box_slot:
                                    VerticalAlignment:=vertical_alignment.Center
                                    HorizontalAlignment:=horizontal_alignment.Center
                                    #Distribution:=option{0.33}
                                    Padding:=margin{Left:=0.0, Top:=0.0, Right:=0.0, Bottom:=0.0}
                                    Widget:=texture_block{DefaultImage:=Textures.General.T_GenericSquare, DefaultDesiredSize:=vector2{X:=1500.0, Y:=2.0}, DefaultTint:=MakeColorFromHex("#2e7be3FF")}

                                #Paging layout
                                stack_box_slot:
                                    VerticalAlignment:=vertical_alignment.Center
                                    HorizontalAlignment:=horizontal_alignment.Center
                                    #Distribution:=option{0.33}
                                    Padding:=margin{Left:=0.0, Top:=0.0, Right:=0.0, Bottom:=0.0}
                                    Widget:=overlay:
                                        Slots:=array:

                                            #Makeshift Size Box
                                            overlay_slot:
                                                VerticalAlignment:=vertical_alignment.Center
                                                HorizontalAlignment:=horizontal_alignment.Center
                                                #Distribution:=option{1.0}
                                                Padding:=margin{Left:=0.0, Top:=0.0, Right:=0.0, Bottom:=0.0}
                                                Widget:=texture_block{DefaultImage:=Textures.General.T_GenSquTrans, DefaultDesiredSize:=vector2{X:=1500.0, Y:=50.0}, DefaultTint:=MakeColorFromHex("#FF0000FF")}

                                            #Paging
                                            overlay_slot:
                                                VerticalAlignment:=vertical_alignment.Top
                                                HorizontalAlignment:=horizontal_alignment.Center
                                                #Distribution:=option{0.8}
                                                Padding:=margin{Left:=0.0, Top:=0.0, Right:=0.0, Bottom:=0.0}
                                                Widget:=stack_box:
                                                    Orientation:=orientation.Horizontal
                                                    Slots:=array:

                                                        #Paging tab texture
                                                        stack_box_slot:
                                                            VerticalAlignment:=vertical_alignment.Top
                                                            HorizontalAlignment:=horizontal_alignment.Center
                                                            Distribution:=option{0.9}
                                                            Padding:=margin{Left:=60.0, Top:=0.0, Right:=0.0, Bottom:=0.0}
                                                            Widget:=overlay:
                                                                Slots:=array:

                                                                    #Makeshift stack box to center the pager
                                                                    overlay_slot:
                                                                        VerticalAlignment:=vertical_alignment.Top
                                                                        HorizontalAlignment:=horizontal_alignment.Center
                                                                        #Distribution:=option{0.33}
                                                                        Padding:=margin{Left:=0.0, Top:=0.0, Right:=0.0, Bottom:=0.0}
                                                                        Widget:=texture_block{DefaultImage:=Textures.General.T_GenSquTrans, DefaultDesiredSize:=vector2{X:=1350.0, Y:=50.0}, DefaultTint:=MakeColorFromHex("#00FF00FF")}

                                                                    overlay_slot:
                                                                        VerticalAlignment:=vertical_alignment.Top
                                                                        HorizontalAlignment:=horizontal_alignment.Center
                                                                        #Distribution:=option{0.33}
                                                                        Padding:=margin{Left:=0.0, Top:=0.0, Right:=0.0, Bottom:=0.0}
                                                                        Widget:=texture_block{DefaultImage:=Textures.General.T_TabDown, DefaultDesiredSize:=vector2{X:=250.0, Y:=50.0}, DefaultTint:=MakeColorFromHex("#2e7be3FF")}

                                                                    #Paging tab texture
                                                                    overlay_slot:
                                                                        VerticalAlignment:=vertical_alignment.Top
                                                                        HorizontalAlignment:=horizontal_alignment.Center
                                                                        #Distribution:=option{0.33}
                                                                        Padding:=margin{Left:=0.0, Top:=0.0, Right:=0.0, Bottom:=0.0}
                                                                        Widget:=text_block{DefaultText := StringToMessage("Page 1 of 3"), DefaultTextColor:=MakeColorFromHex("#abc1e5FF"), DefaultShadowColor:=MakeColorFromHex("#000000FF"), DefaultShadowOffset:=option{vector2{X:=3.0, Y:=3.0}}}

                                                        #Navigation buttons
                                                        stack_box_slot:
                                                            VerticalAlignment:=vertical_alignment.Center
                                                            HorizontalAlignment:=horizontal_alignment.Right
                                                            Distribution:=option{0.1}
                                                            Padding:=margin{Left:=0.0, Top:=5.0, Right:=0.0, Bottom:=0.0}
                                                            Widget:=stack_box:
                                                                Orientation:=orientation.Horizontal
                                                                Slots:=array:

                                                                    #Nav left
                                                                    stack_box_slot:
                                                                        VerticalAlignment:=vertical_alignment.Center
                                                                        HorizontalAlignment:=horizontal_alignment.Center
                                                                        #Distribution:=option{0.3}
                                                                        Padding:=margin{Left:=0.0, Top:=0.0, Right:=3.0, Bottom:=0.0}
                                                                        Widget:=button_loud{DefaultText:=StringToMessage("<")}

                                                                    #Nav right
                                                                    stack_box_slot:
                                                                        VerticalAlignment:=vertical_alignment.Center
                                                                        HorizontalAlignment:=horizontal_alignment.Center
                                                                        #Distribution:=option{0.3}
                                                                        Padding:=margin{Left:=0.0, Top:=0.0, Right:=0.0, Bottom:=0.0}
                                                                        Widget:=button_loud{DefaultText:=StringToMessage(">")}

            #Main content Layout
            NavPurchase1:=overlay:
                Slots:=array:

                    #Makeshift size box value
                    overlay_slot:
                        VerticalAlignment:=vertical_alignment.Center
                        HorizontalAlignment:=horizontal_alignment.Center
                        #Distribution:=option{0.33}
                        Padding:=margin{Left:=0.0, Top:=0.0, Right:=0.0, Bottom:=0.0}
                        Widget:=texture_block{DefaultImage:=Textures.General.T_GenSquTrans, DefaultDesiredSize:=vector2{X:=1800.0, Y:=760.0}, DefaultTint:=MakeColorFromHex("#00GG00FF")}

                    #Main content structure
                    overlay_slot:
                        VerticalAlignment:=vertical_alignment.Center
                        HorizontalAlignment:=horizontal_alignment.Center
                        #Distribution:=option{0.33}
                        Padding:=margin{Left:=0.0, Top:=0.0, Right:=0.0, Bottom:=0.0}
                        Widget:=stack_box:
                            Orientation:=orientation.Vertical
                            Slots:=array:

                                #Single row layout
                                stack_box_slot:
                                    VerticalAlignment:=vertical_alignment.Center
                                    HorizontalAlignment:=horizontal_alignment.Center
                                    #Distribution:=option{0.33}
                                    Padding:=margin{Left:=0.0, Top:=0.0, Right:=0.0, Bottom:=0.0}
                                    Widget:=overlay:
                                        Slots:=array:
                                            
                                            #Makeshift size box
                                                overlay_slot:
                                                    VerticalAlignment:=vertical_alignment.Center
                                                    HorizontalAlignment:=horizontal_alignment.Center
                                                    #Distribution:=option{0.33}
                                                    Padding:=margin{Left:=0.0, Top:=0.0, Right:=0.0, Bottom:=0.0}
                                                    Widget:=texture_block{DefaultImage:=Textures.General.T_GenSquTrans, DefaultDesiredSize:=vector2{X:=1600.0, Y:=120.0}, DefaultTint:=MakeColorFromHex("#FF0000FF")}        

                                            #Row layout
                                            overlay_slot:
                                                VerticalAlignment:=vertical_alignment.Center
                                                HorizontalAlignment:=horizontal_alignment.Fill
                                                #Distribution:=option{0.33}
                                                Padding:=margin{Left:=0.0, Top:=0.0, Right:=0.0, Bottom:=0.0}
                                                Widget:=stack_box:
                                                    Orientation:=orientation.Horizontal
                                                    Slots:=array:
                                                        
                                                        #Booster icon and rarity
                                                        stack_box_slot:
                                                            VerticalAlignment:=vertical_alignment.Center
                                                            HorizontalAlignment:=horizontal_alignment.Left
                                                            Distribution:=option{1.0}
                                                            Padding:=margin{Left:=0.0, Top:=0.0, Right:=0.0, Bottom:=0.0}
                                                            Widget:=overlay:
                                                                Slots:=array:
                                                                    
                                                                    #Booster Rarity BG
                                                                    overlay_slot:
                                                                        VerticalAlignment:=vertical_alignment.Center
                                                                        HorizontalAlignment:=horizontal_alignment.Left
                                                                        #Distribution:=option{1.0}
                                                                        Padding:=margin{Left:=0.0, Top:=0.0, Right:=0.0, Bottom:=0.0}
                                                                        Widget:=texture_block{DefaultImage:=Textures.Rarity.T_IconBG_Common, DefaultDesiredSize:=vector2{X:=120.0, Y:=120.0}, DefaultTint:=MakeColorFromHex("#FFFFFFFF")}
                                                                    
                                                                    #Booster Icon
                                                                    overlay_slot:
                                                                        VerticalAlignment:=vertical_alignment.Center
                                                                        HorizontalAlignment:=horizontal_alignment.Left
                                                                        #Distribution:=option{1.0}
                                                                        Padding:=margin{Left:=0.0, Top:=0.0, Right:=0.0, Bottom:=0.0}
                                                                        Widget:=texture_block{DefaultImage:=Textures.Boosters.T_IncreaseMaxShield, DefaultDesiredSize:=vector2{X:=116.0, Y:=116.0}, DefaultTint:=MakeColorFromHex("#FFFFFFFF")}

                                                                    #Booster Border
                                                                    overlay_slot:
                                                                        VerticalAlignment:=vertical_alignment.Center
                                                                        HorizontalAlignment:=horizontal_alignment.Left
                                                                        #Distribution:=option{1.0}
                                                                        Padding:=margin{Left:=0.0, Top:=0.0, Right:=0.0, Bottom:=0.0}
                                                                        Widget:=texture_block{DefaultImage:=Textures.General.T_GenericBorderSquare, DefaultDesiredSize:=vector2{X:=120.0, Y:=120.0}, DefaultTint:=MakeColorFromHex("#cdc953FF")}
                                                                                                                                                                                                        
                                                        #Booster description
                                                        stack_box_slot:
                                                            VerticalAlignment:=vertical_alignment.Center
                                                            HorizontalAlignment:=horizontal_alignment.Fill
                                                            #Distribution:=option{1.0}
                                                            Padding:=margin{Left:=0.0, Top:=0.0, Right:=0.0, Bottom:=0.0}
                                                            Widget:=overlay:
                                                                Slots:=array:

                                                                    #BG color
                                                                    overlay_slot:
                                                                        VerticalAlignment:=vertical_alignment.Center
                                                                        HorizontalAlignment:=horizontal_alignment.Center
                                                                        #Distribution:=option{0.33}
                                                                        Padding:=margin{Left:=0.0, Top:=0.0, Right:=0.0, Bottom:=0.0}
                                                                        Widget:=texture_block{DefaultImage:=Textures.General.T_GenericSquare, DefaultDesiredSize:=vector2{X:=570.0, Y:=120.0}, DefaultTint:=MakeColorFromHex("#031126FF")}

                                                                    #Overall Border
                                                                    overlay_slot:
                                                                        VerticalAlignment:=vertical_alignment.Center
                                                                        HorizontalAlignment:=horizontal_alignment.Center
                                                                        #Distribution:=option{0.33}
                                                                        Padding:=margin{Left:=0.0, Top:=0.0, Right:=0.0, Bottom:=0.0}
                                                                        Widget:=texture_block{DefaultImage:=Textures.General.T_LevelBorder, DefaultDesiredSize:=vector2{X:=570.0, Y:=120.0}, DefaultTint:=MakeColorFromHex("#1d4b89FF")}
                                                                    
                                                                        
                                                                        
                                                                    #Description Header
                                                                    overlay_slot:
                                                                        VerticalAlignment:=vertical_alignment.Top
                                                                        HorizontalAlignment:=horizontal_alignment.Center
                                                                        #Distribution:=option{0.33}
                                                                        Padding:=margin{Left:=0.0, Top:=0.0, Right:=0.0, Bottom:=5.0}
                                                                        Widget:=text_block{DefaultText := StringToMessage("Increase Max Shield"), DefaultTextColor:=MakeColorFromHex("#f6fe17FF"), DefaultShadowColor:=MakeColorFromHex("#000000FF"), DefaultShadowOffset:=option{vector2{X:=3.0, Y:=3.0}}}

                                                                    #Description Header
                                                                    overlay_slot:
                                                                        VerticalAlignment:=vertical_alignment.Bottom
                                                                        HorizontalAlignment:=horizontal_alignment.Center
                                                                        #Distribution:=option{0.33}
                                                                        Padding:=margin{Left:=0.0, Top:=0.0, Right:=0.0, Bottom:=0.0}
                                                                        Widget:=text_block{DefaultText := StringToMessage("If health falls under 40%,\nrecover 50% shield once"), DefaultTextColor:=MakeColorFromHex("#c4c5c7FF"), DefaultShadowColor:=MakeColorFromHex("#000000FF"), DefaultShadowOffset:=option{vector2{X:=3.0, Y:=3.0}}}

                                                        #Upgrade description
                                                        stack_box_slot:
                                                            VerticalAlignment:=vertical_alignment.Center
                                                            HorizontalAlignment:=horizontal_alignment.Center
                                                            #Distribution:=option{0.33}
                                                            Padding:=margin{Left:=0.0, Top:=0.0, Right:=0.0, Bottom:=0.0}
                                                            Widget:=overlay:
                                                                Slots:=array:

                                                                    #BG color
                                                                    overlay_slot:
                                                                        VerticalAlignment:=vertical_alignment.Center
                                                                        HorizontalAlignment:=horizontal_alignment.Center
                                                                        #Distribution:=option{0.33}
                                                                        Padding:=margin{Left:=0.0, Top:=0.0, Right:=0.0, Bottom:=0.0}
                                                                        Widget:=texture_block{DefaultImage:=Textures.General.T_GenericSquare, DefaultDesiredSize:=vector2{X:=348.0, Y:=120.0}, DefaultTint:=MakeColorFromHex("#000000FF")}

                                                                    #Overall Border
                                                                    overlay_slot:
                                                                        VerticalAlignment:=vertical_alignment.Center
                                                                        HorizontalAlignment:=horizontal_alignment.Center
                                                                        #Distribution:=option{0.33}
                                                                        Padding:=margin{Left:=0.0, Top:=0.0, Right:=0.0, Bottom:=0.0}
                                                                        Widget:=texture_block{DefaultImage:=Textures.General.T_LevelBorder, DefaultDesiredSize:=vector2{X:=348.0, Y:=120.0}, DefaultTint:=MakeColorFromHex("#1d4b89FF")}

                                                                    #Upgrade description layout - Horizontal stack box
                                                                    overlay_slot:
                                                                        VerticalAlignment:=vertical_alignment.Center
                                                                        HorizontalAlignment:=horizontal_alignment.Fill
                                                                        #Distribution:=option{1.0}
                                                                        Padding:=margin{Left:=0.0, Top:=0.0, Right:=0.0, Bottom:=5.0}
                                                                        Widget:=stack_box:
                                                                            Orientation:=orientation.Horizontal
                                                                            Slots:=array:

                                                                                #First description column
                                                                                stack_box_slot:
                                                                                    VerticalAlignment:=vertical_alignment.Center
                                                                                    HorizontalAlignment:=horizontal_alignment.Left
                                                                                    Distribution:=option{1.0}
                                                                                    Padding:=margin{Left:=25.0, Top:=0.0, Right:=0.0, Bottom:=0.0}
                                                                                    Widget:=stack_box:
                                                                                        Orientation:=orientation.Vertical
                                                                                        Slots:=array:

                                                                                            #First horizontal stack box
                                                                                            stack_box_slot:
                                                                                                VerticalAlignment:=vertical_alignment.Center
                                                                                                HorizontalAlignment:=horizontal_alignment.Center
                                                                                                Distribution:=option{1.0}
                                                                                                Padding:=margin{Left:=0.0, Top:=0.0, Right:=0.0, Bottom:=5.0}
                                                                                                Widget:=stack_box:
                                                                                                    Orientation:=orientation.Horizontal
                                                                                                    Slots:=array:

                                                                                                        #Base Lv text
                                                                                                        stack_box_slot:
                                                                                                            VerticalAlignment:=vertical_alignment.Center
                                                                                                            HorizontalAlignment:=horizontal_alignment.Center
                                                                                                            #Distribution:=option{0.33}
                                                                                                            Padding:=margin{Left:=0.0, Top:=5.0, Right:=8.0, Bottom:=0.0}
                                                                                                            Widget:=text_block{DefaultText := StringToMessage("BASE LV"), DefaultTextColor:=MakeColorFromHex("#FFFFFFFF"), DefaultShadowColor:=MakeColorFromHex("#000000FF"), DefaultShadowOffset:=option{vector2{X:=3.0, Y:=3.0}}}

                                                                                            #Second horizontal stack box (Actually an overlay)
                                                                                            stack_box_slot:
                                                                                                VerticalAlignment:=vertical_alignment.Center
                                                                                                HorizontalAlignment:=horizontal_alignment.Center
                                                                                                Distribution:=option{1.0}
                                                                                                Padding:=margin{Left:=0.0, Top:=0.0, Right:=0.0, Bottom:=0.0}
                                                                                                Widget:=overlay:
                                                                                                    Slots:=array:

                                                                                                        #Base description text 1
                                                                                                        overlay_slot:
                                                                                                            VerticalAlignment:=vertical_alignment.Center
                                                                                                            HorizontalAlignment:=horizontal_alignment.Center
                                                                                                            #Distribution:=option{0.33}
                                                                                                            Padding:=margin{Left:=0.0, Top:=0.0, Right:=0.0, Bottom:=0.0}
                                                                                                            Widget:=stack_box:
                                                                                                                Orientation:=orientation.Horizontal
                                                                                                                Slots:=array:

                                                                                                                    stack_box_slot:
                                                                                                                        VerticalAlignment:=vertical_alignment.Center
                                                                                                                        HorizontalAlignment:=horizontal_alignment.Center
                                                                                                                        #Distribution:=option{0.33}
                                                                                                                        Padding:=margin{Left:=0.0, Top:=0.0, Right:=7.0, Bottom:=5.0}
                                                                                                                        Widget:=texture_block{DefaultImage:=Textures.Boosters.T_IncreaseMaxShield, DefaultDesiredSize:=vector2{X:=25.0, Y:=25.0}, DefaultTint:=MakeColorFromHex("#FFFFFFFF")}

                                                                                                                    stack_box_slot:
                                                                                                                        VerticalAlignment:=vertical_alignment.Center
                                                                                                                        HorizontalAlignment:=horizontal_alignment.Center
                                                                                                                        #Distribution:=option{0.33}
                                                                                                                        Padding:=margin{Left:=0.0, Top:=0.0, Right:=20.0, Bottom:=5.0}
                                                                                                                        Widget:=text_block{DefaultText := StringToMessage("+50%"), DefaultTextColor:=MakeColorFromHex("#c3c3c3FF"), DefaultShadowColor:=MakeColorFromHex("#000000FF"), DefaultShadowOffset:=option{vector2{X:=3.0, Y:=3.0}}}
                                                                                                        
                                                                                                        #Base description text 2
                                                                                                        overlay_slot:
                                                                                                            VerticalAlignment:=vertical_alignment.Bottom
                                                                                                            HorizontalAlignment:=horizontal_alignment.Center
                                                                                                            #Distribution:=option{0.33}
                                                                                                            Padding:=margin{Left:=0.0, Top:=20.0, Right:=0.0, Bottom:=0.0}
                                                                                                            Widget:=stack_box:
                                                                                                                Orientation:=orientation.Horizontal
                                                                                                                Slots:=array:

                                                                                                                    stack_box_slot:
                                                                                                                        VerticalAlignment:=vertical_alignment.Center
                                                                                                                        HorizontalAlignment:=horizontal_alignment.Fill
                                                                                                                        Distribution:=option{1.0}
                                                                                                                        Padding:=margin{Left:=0.0, Top:=0.0, Right:=30.0, Bottom:=0.0}
                                                                                                                        Widget:=text_block{DefaultText := StringToMessage("once"), DefaultTextColor:=MakeColorFromHex("#c3c3c3FF"), DefaultShadowColor:=MakeColorFromHex("#000000FF"), DefaultShadowOffset:=option{vector2{X:=3.0, Y:=3.0}}}

                                                                                #Arrow
                                                                                stack_box_slot:
                                                                                    VerticalAlignment:=vertical_alignment.Center
                                                                                    HorizontalAlignment:=horizontal_alignment.Center
                                                                                    #Distribution:=option{0.33}
                                                                                    Padding:=margin{Left:=8.0, Top:=0.0, Right:=12.0, Bottom:=3.0}
                                                                                    Widget:=overlay:
                                                                                        Slots:=array:
                                                                                            
                                                                                            #Arrow
                                                                                            overlay_slot:
                                                                                                VerticalAlignment:=vertical_alignment.Center
                                                                                                HorizontalAlignment:=horizontal_alignment.Center
                                                                                                #Distribution:=option{0.33}
                                                                                                Padding:=margin{Left:=0.0, Top:=0.0, Right:=0.0, Bottom:=3.0}
                                                                                                Widget:=texture_block{DefaultImage:=Textures.General.T_GenArrowHead, DefaultDesiredSize:=vector2{X:=26.0, Y:=24.0}, DefaultTint:=MakeColorFromHex("#FFFFFFFF")}
                                                                                
                                                                                #Second description column
                                                                                stack_box_slot:
                                                                                    VerticalAlignment:=vertical_alignment.Center
                                                                                    HorizontalAlignment:=horizontal_alignment.Right
                                                                                    Distribution:=option{1.0}
                                                                                    Padding:=margin{Left:=0.0, Top:=0.0, Right:=0.0, Bottom:=0.0}
                                                                                    Widget:=stack_box:
                                                                                        Orientation:=orientation.Vertical
                                                                                        Slots:=array:

                                                                                            #First horizontal stack box
                                                                                            stack_box_slot:
                                                                                                VerticalAlignment:=vertical_alignment.Center
                                                                                                HorizontalAlignment:=horizontal_alignment.Center
                                                                                                Distribution:=option{1.0}
                                                                                                Padding:=margin{Left:=0.0, Top:=0.0, Right:=0.0, Bottom:=5.0}
                                                                                                Widget:=stack_box:
                                                                                                    Orientation:=orientation.Horizontal
                                                                                                    Slots:=array:

                                                                                                        #Lv 1 text
                                                                                                        stack_box_slot:
                                                                                                            VerticalAlignment:=vertical_alignment.Center
                                                                                                            HorizontalAlignment:=horizontal_alignment.Center
                                                                                                            #Distribution:=option{0.33}
                                                                                                            Padding:=margin{Left:=0.0, Top:=5.0, Right:=8.0, Bottom:=0.0}
                                                                                                            Widget:=text_block{DefaultText := StringToMessage("LV. 1"), DefaultTextColor:=MakeColorFromHex("#f6fe17FF"), DefaultShadowColor:=MakeColorFromHex("#000000FF"), DefaultShadowOffset:=option{vector2{X:=3.0, Y:=3.0}}}
                                                                                                        
                                                                                            #Second horizontal stack box (Actually an overlay)
                                                                                            stack_box_slot:
                                                                                                VerticalAlignment:=vertical_alignment.Center
                                                                                                HorizontalAlignment:=horizontal_alignment.Center
                                                                                                Distribution:=option{1.0}
                                                                                                Padding:=margin{Left:=0.0, Top:=0.0, Right:=0.0, Bottom:=0.0}
                                                                                                Widget:=overlay:
                                                                                                    Slots:=array:

                                                                                                        #Base description text 1
                                                                                                        overlay_slot:
                                                                                                            VerticalAlignment:=vertical_alignment.Center
                                                                                                            HorizontalAlignment:=horizontal_alignment.Center
                                                                                                            #Distribution:=option{0.33}
                                                                                                            Padding:=margin{Left:=0.0, Top:=0.0, Right:=0.0, Bottom:=0.0}
                                                                                                            Widget:=stack_box:
                                                                                                                Orientation:=orientation.Horizontal
                                                                                                                Slots:=array:

                                                                                                                    stack_box_slot:
                                                                                                                        VerticalAlignment:=vertical_alignment.Center
                                                                                                                        HorizontalAlignment:=horizontal_alignment.Center
                                                                                                                        #Distribution:=option{0.33}
                                                                                                                        Padding:=margin{Left:=0.0, Top:=0.0, Right:=7.0, Bottom:=5.0}
                                                                                                                        Widget:=texture_block{DefaultImage:=Textures.Boosters.T_IncreaseMaxShield, DefaultDesiredSize:=vector2{X:=25.0, Y:=25.0}, DefaultTint:=MakeColorFromHex("#FFFFFFFF")}

                                                                                                                    stack_box_slot:
                                                                                                                        VerticalAlignment:=vertical_alignment.Center
                                                                                                                        HorizontalAlignment:=horizontal_alignment.Center
                                                                                                                        #Distribution:=option{0.33}
                                                                                                                        Padding:=margin{Left:=0.0, Top:=0.0, Right:=20.0, Bottom:=5.0}
                                                                                                                        Widget:=text_block{DefaultText := StringToMessage("+50%"), DefaultTextColor:=MakeColorFromHex("#f6fe17FF"), DefaultShadowColor:=MakeColorFromHex("#000000FF"), DefaultShadowOffset:=option{vector2{X:=3.0, Y:=3.0}}}
                                                                                                                    
                                                                                                        #Base description text 2
                                                                                                        overlay_slot:
                                                                                                            VerticalAlignment:=vertical_alignment.Bottom
                                                                                                            HorizontalAlignment:=horizontal_alignment.Center
                                                                                                            #Distribution:=option{0.33}
                                                                                                            Padding:=margin{Left:=0.0, Top:=20.0, Right:=0.0, Bottom:=0.0}
                                                                                                            Widget:=stack_box:
                                                                                                                Orientation:=orientation.Horizontal
                                                                                                                Slots:=array:

                                                                                                                    stack_box_slot:
                                                                                                                        VerticalAlignment:=vertical_alignment.Center
                                                                                                                        HorizontalAlignment:=horizontal_alignment.Left
                                                                                                                        Distribution:=option{1.0}
                                                                                                                        Padding:=margin{Left:=0.0, Top:=0.0, Right:=30.0, Bottom:=0.0}
                                                                                                                        Widget:=text_block{DefaultText := StringToMessage("twice"), DefaultTextColor:=MakeColorFromHex("#f6fe17FF"), DefaultShadowColor:=MakeColorFromHex("#000000FF"), DefaultShadowOffset:=option{vector2{X:=3.0, Y:=3.0}}}
                                                                                                                    
                                                        #Spacing
                                                        stack_box_slot:
                                                            VerticalAlignment:=vertical_alignment.Center
                                                            HorizontalAlignment:=horizontal_alignment.Right
                                                            #Distribution:=option{1.0}
                                                            Padding:=margin{Left:=0.0, Top:=0.0, Right:=0.0, Bottom:=0.0}
                                                            Widget:=texture_block{DefaultImage:=Textures.General.T_GenSquTrans, DefaultDesiredSize:=vector2{X:=350.0, Y:=120.0}, DefaultTint:=MakeColorFromHex("#00FF00FF")}
                                                        
                                                        #Call to action
                                                        stack_box_slot:
                                                            VerticalAlignment:=vertical_alignment.Center
                                                            HorizontalAlignment:=horizontal_alignment.Right
                                                            #Distribution:=option{1.0}
                                                            Padding:=margin{Left:=0.0, Top:=0.0, Right:=0.0, Bottom:=0.0}
                                                            Widget:=overlay:
                                                                Slots:=array:

                                                                    #Overlay structure, first is the button itself
                                                                    overlay_slot:
                                                                        VerticalAlignment:=vertical_alignment.Center
                                                                        HorizontalAlignment:=horizontal_alignment.Center
                                                                        #Distribution:=option{1.0}
                                                                        Padding:=margin{Left:=0.0, Top:=0.0, Right:=0.0, Bottom:=0.0}
                                                                        Widget:=button_loud{DefaultText:=StringToMessage("UPGRADE\n ")}

                                                                    #Overlay structure, icon and value
                                                                    overlay_slot:
                                                                        VerticalAlignment:=vertical_alignment.Fill
                                                                        HorizontalAlignment:=horizontal_alignment.Center
                                                                        #Distribution:=option{1.0}
                                                                        Padding:=margin{Left:=0.0, Top:=0.0, Right:=0.0, Bottom:=0.0}
                                                                        Widget:=stack_box:
                                                                            Orientation:=orientation.Horizontal
                                                                            Slots:=array:

                                                                                #Icon and value
                                                                                stack_box_slot:
                                                                                    VerticalAlignment:=vertical_alignment.Bottom
                                                                                    HorizontalAlignment:=horizontal_alignment.Right
                                                                                    #Distribution:=option{1.0}
                                                                                    Padding:=margin{Left:=0.0, Top:=0.0, Right:=0.0, Bottom:=8.0}
                                                                                    Widget:=texture_block{DefaultImage:=Textures.Currency.T_Gold, DefaultDesiredSize:=vector2{X:=35.0, Y:=35.0}, DefaultTint:=MakeColorFromHex("#000000FF")}                

                                                                                stack_box_slot:
                                                                                    VerticalAlignment:=vertical_alignment.Bottom
                                                                                    HorizontalAlignment:=horizontal_alignment.Center
                                                                                    #Distribution:=option{1.0}
                                                                                    Padding:=margin{Left:=5.0, Top:=0.0, Right:=0.0, Bottom:=5.0}
                                                                                    Widget:=text_block{DefaultText := StringToMessage("15"), DefaultTextColor:=MakeColorFromHex("#000000FF"), DefaultShadowColor:=MakeColorFromHex("#000000FF"), DefaultShadowOffset:=option{vector2{X:=3.0, Y:=3.0}}}
                                                                                                                
            NavPurchase2:=overlay:
                Slots:=array:

                    #Makeshift size box value
                    overlay_slot:
                        VerticalAlignment:=vertical_alignment.Center
                        HorizontalAlignment:=horizontal_alignment.Center
                        #Distribution:=option{0.33}
                        Padding:=margin{Left:=0.0, Top:=0.0, Right:=0.0, Bottom:=0.0}
                        Widget:=texture_block{DefaultImage:=Textures.General.T_GenSquTrans, DefaultDesiredSize:=vector2{X:=1800.0, Y:=760.0}, DefaultTint:=MakeColorFromHex("#00GG00FF")}

                    #Main content structure
                    overlay_slot:
                        VerticalAlignment:=vertical_alignment.Center
                        HorizontalAlignment:=horizontal_alignment.Center
                        #Distribution:=option{0.33}
                        Padding:=margin{Left:=0.0, Top:=0.0, Right:=0.0, Bottom:=0.0}
                        Widget:=stack_box:
                            Orientation:=orientation.Vertical
                            Slots:=array:

                                #Single row layout
                                stack_box_slot:
                                    VerticalAlignment:=vertical_alignment.Center
                                    HorizontalAlignment:=horizontal_alignment.Center
                                    #Distribution:=option{0.33}
                                    Padding:=margin{Left:=0.0, Top:=0.0, Right:=0.0, Bottom:=0.0}
                                    Widget:=overlay:
                                        Slots:=array:
                                            
                                            #Makeshift size box
                                                overlay_slot:
                                                    VerticalAlignment:=vertical_alignment.Center
                                                    HorizontalAlignment:=horizontal_alignment.Center
                                                    #Distribution:=option{0.33}
                                                    Padding:=margin{Left:=0.0, Top:=0.0, Right:=0.0, Bottom:=0.0}
                                                    Widget:=texture_block{DefaultImage:=Textures.General.T_GenSquTrans, DefaultDesiredSize:=vector2{X:=1600.0, Y:=120.0}, DefaultTint:=MakeColorFromHex("#FF0000FF")}        

                                            #Row layout
                                            overlay_slot:
                                                VerticalAlignment:=vertical_alignment.Center
                                                HorizontalAlignment:=horizontal_alignment.Fill
                                                #Distribution:=option{0.33}
                                                Padding:=margin{Left:=0.0, Top:=0.0, Right:=0.0, Bottom:=0.0}
                                                Widget:=stack_box:
                                                    Orientation:=orientation.Horizontal
                                                    Slots:=array:
                                                        
                                                        #Booster icon and rarity
                                                        stack_box_slot:
                                                            VerticalAlignment:=vertical_alignment.Center
                                                            HorizontalAlignment:=horizontal_alignment.Left
                                                            Distribution:=option{1.0}
                                                            Padding:=margin{Left:=0.0, Top:=0.0, Right:=0.0, Bottom:=0.0}
                                                            Widget:=overlay:
                                                                Slots:=array:
                                                                    
                                                                    #Booster Rarity BG
                                                                    overlay_slot:
                                                                        VerticalAlignment:=vertical_alignment.Center
                                                                        HorizontalAlignment:=horizontal_alignment.Left
                                                                        #Distribution:=option{1.0}
                                                                        Padding:=margin{Left:=0.0, Top:=0.0, Right:=0.0, Bottom:=0.0}
                                                                        Widget:=texture_block{DefaultImage:=Textures.Rarity.T_IconBG_Common, DefaultDesiredSize:=vector2{X:=120.0, Y:=120.0}, DefaultTint:=MakeColorFromHex("#FFFFFFFF")}
                                                                    
                                                                    #Booster Icon
                                                                    overlay_slot:
                                                                        VerticalAlignment:=vertical_alignment.Center
                                                                        HorizontalAlignment:=horizontal_alignment.Left
                                                                        #Distribution:=option{1.0}
                                                                        Padding:=margin{Left:=0.0, Top:=0.0, Right:=0.0, Bottom:=0.0}
                                                                        Widget:=texture_block{DefaultImage:=Textures.Boosters.T_IncreaseMaxShield, DefaultDesiredSize:=vector2{X:=116.0, Y:=116.0}, DefaultTint:=MakeColorFromHex("#FFFFFFFF")}

                                                                    #Booster Border
                                                                    overlay_slot:
                                                                        VerticalAlignment:=vertical_alignment.Center
                                                                        HorizontalAlignment:=horizontal_alignment.Left
                                                                        #Distribution:=option{1.0}
                                                                        Padding:=margin{Left:=0.0, Top:=0.0, Right:=0.0, Bottom:=0.0}
                                                                        Widget:=texture_block{DefaultImage:=Textures.General.T_GenericBorderSquare, DefaultDesiredSize:=vector2{X:=120.0, Y:=120.0}, DefaultTint:=MakeColorFromHex("#cdc953FF")}
                                                                                                                                                                                                        
                                                        #Booster description
                                                        stack_box_slot:
                                                            VerticalAlignment:=vertical_alignment.Center
                                                            HorizontalAlignment:=horizontal_alignment.Fill
                                                            #Distribution:=option{1.0}
                                                            Padding:=margin{Left:=0.0, Top:=0.0, Right:=0.0, Bottom:=0.0}
                                                            Widget:=overlay:
                                                                Slots:=array:

                                                                    #BG color
                                                                    overlay_slot:
                                                                        VerticalAlignment:=vertical_alignment.Center
                                                                        HorizontalAlignment:=horizontal_alignment.Center
                                                                        #Distribution:=option{0.33}
                                                                        Padding:=margin{Left:=0.0, Top:=0.0, Right:=0.0, Bottom:=0.0}
                                                                        Widget:=texture_block{DefaultImage:=Textures.General.T_GenericSquare, DefaultDesiredSize:=vector2{X:=570.0, Y:=120.0}, DefaultTint:=MakeColorFromHex("#031126FF")}

                                                                    #Overall Border
                                                                    overlay_slot:
                                                                        VerticalAlignment:=vertical_alignment.Center
                                                                        HorizontalAlignment:=horizontal_alignment.Center
                                                                        #Distribution:=option{0.33}
                                                                        Padding:=margin{Left:=0.0, Top:=0.0, Right:=0.0, Bottom:=0.0}
                                                                        Widget:=texture_block{DefaultImage:=Textures.General.T_LevelBorder, DefaultDesiredSize:=vector2{X:=570.0, Y:=120.0}, DefaultTint:=MakeColorFromHex("#1d4b89FF")}
                                                                    
                                                                        
                                                                        
                                                                    #Description Header
                                                                    overlay_slot:
                                                                        VerticalAlignment:=vertical_alignment.Top
                                                                        HorizontalAlignment:=horizontal_alignment.Center
                                                                        #Distribution:=option{0.33}
                                                                        Padding:=margin{Left:=0.0, Top:=0.0, Right:=0.0, Bottom:=5.0}
                                                                        Widget:=text_block{DefaultText := StringToMessage("Increase Max Shield"), DefaultTextColor:=MakeColorFromHex("#f6fe17FF"), DefaultShadowColor:=MakeColorFromHex("#000000FF"), DefaultShadowOffset:=option{vector2{X:=3.0, Y:=3.0}}}

                                                                    #Description Header
                                                                    overlay_slot:
                                                                        VerticalAlignment:=vertical_alignment.Bottom
                                                                        HorizontalAlignment:=horizontal_alignment.Center
                                                                        #Distribution:=option{0.33}
                                                                        Padding:=margin{Left:=0.0, Top:=0.0, Right:=0.0, Bottom:=0.0}
                                                                        Widget:=text_block{DefaultText := StringToMessage("If health falls under 40%,\nrecover 50% shield once"), DefaultTextColor:=MakeColorFromHex("#c4c5c7FF"), DefaultShadowColor:=MakeColorFromHex("#000000FF"), DefaultShadowOffset:=option{vector2{X:=3.0, Y:=3.0}}}

                                                        #Upgrade description
                                                        stack_box_slot:
                                                            VerticalAlignment:=vertical_alignment.Center
                                                            HorizontalAlignment:=horizontal_alignment.Center
                                                            #Distribution:=option{0.33}
                                                            Padding:=margin{Left:=0.0, Top:=0.0, Right:=0.0, Bottom:=0.0}
                                                            Widget:=overlay:
                                                                Slots:=array:

                                                                    #BG color
                                                                    overlay_slot:
                                                                        VerticalAlignment:=vertical_alignment.Center
                                                                        HorizontalAlignment:=horizontal_alignment.Center
                                                                        #Distribution:=option{0.33}
                                                                        Padding:=margin{Left:=0.0, Top:=0.0, Right:=0.0, Bottom:=0.0}
                                                                        Widget:=texture_block{DefaultImage:=Textures.General.T_GenericSquare, DefaultDesiredSize:=vector2{X:=348.0, Y:=120.0}, DefaultTint:=MakeColorFromHex("#000000FF")}

                                                                    #Overall Border
                                                                    overlay_slot:
                                                                        VerticalAlignment:=vertical_alignment.Center
                                                                        HorizontalAlignment:=horizontal_alignment.Center
                                                                        #Distribution:=option{0.33}
                                                                        Padding:=margin{Left:=0.0, Top:=0.0, Right:=0.0, Bottom:=0.0}
                                                                        Widget:=texture_block{DefaultImage:=Textures.General.T_LevelBorder, DefaultDesiredSize:=vector2{X:=348.0, Y:=120.0}, DefaultTint:=MakeColorFromHex("#1d4b89FF")}

                                                                    #Upgrade description layout - Horizontal stack box
                                                                    overlay_slot:
                                                                        VerticalAlignment:=vertical_alignment.Center
                                                                        HorizontalAlignment:=horizontal_alignment.Fill
                                                                        #Distribution:=option{1.0}
                                                                        Padding:=margin{Left:=0.0, Top:=0.0, Right:=0.0, Bottom:=5.0}
                                                                        Widget:=stack_box:
                                                                            Orientation:=orientation.Horizontal
                                                                            Slots:=array:

                                                                                #First description column
                                                                                stack_box_slot:
                                                                                    VerticalAlignment:=vertical_alignment.Center
                                                                                    HorizontalAlignment:=horizontal_alignment.Left
                                                                                    Distribution:=option{1.0}
                                                                                    Padding:=margin{Left:=25.0, Top:=0.0, Right:=0.0, Bottom:=0.0}
                                                                                    Widget:=stack_box:
                                                                                        Orientation:=orientation.Vertical
                                                                                        Slots:=array:

                                                                                            #First horizontal stack box
                                                                                            stack_box_slot:
                                                                                                VerticalAlignment:=vertical_alignment.Center
                                                                                                HorizontalAlignment:=horizontal_alignment.Center
                                                                                                Distribution:=option{1.0}
                                                                                                Padding:=margin{Left:=0.0, Top:=0.0, Right:=0.0, Bottom:=5.0}
                                                                                                Widget:=stack_box:
                                                                                                    Orientation:=orientation.Horizontal
                                                                                                    Slots:=array:

                                                                                                        #Base Lv text
                                                                                                        stack_box_slot:
                                                                                                            VerticalAlignment:=vertical_alignment.Center
                                                                                                            HorizontalAlignment:=horizontal_alignment.Center
                                                                                                            #Distribution:=option{0.33}
                                                                                                            Padding:=margin{Left:=0.0, Top:=5.0, Right:=8.0, Bottom:=0.0}
                                                                                                            Widget:=text_block{DefaultText := StringToMessage("BASE LV"), DefaultTextColor:=MakeColorFromHex("#FFFFFFFF"), DefaultShadowColor:=MakeColorFromHex("#000000FF"), DefaultShadowOffset:=option{vector2{X:=3.0, Y:=3.0}}}

                                                                                            #Second horizontal stack box (Actually an overlay)
                                                                                            stack_box_slot:
                                                                                                VerticalAlignment:=vertical_alignment.Center
                                                                                                HorizontalAlignment:=horizontal_alignment.Center
                                                                                                Distribution:=option{1.0}
                                                                                                Padding:=margin{Left:=0.0, Top:=0.0, Right:=0.0, Bottom:=0.0}
                                                                                                Widget:=overlay:
                                                                                                    Slots:=array:

                                                                                                        #Base description text 1
                                                                                                        overlay_slot:
                                                                                                            VerticalAlignment:=vertical_alignment.Center
                                                                                                            HorizontalAlignment:=horizontal_alignment.Center
                                                                                                            #Distribution:=option{0.33}
                                                                                                            Padding:=margin{Left:=0.0, Top:=0.0, Right:=0.0, Bottom:=0.0}
                                                                                                            Widget:=stack_box:
                                                                                                                Orientation:=orientation.Horizontal
                                                                                                                Slots:=array:

                                                                                                                    stack_box_slot:
                                                                                                                        VerticalAlignment:=vertical_alignment.Center
                                                                                                                        HorizontalAlignment:=horizontal_alignment.Center
                                                                                                                        #Distribution:=option{0.33}
                                                                                                                        Padding:=margin{Left:=0.0, Top:=0.0, Right:=7.0, Bottom:=5.0}
                                                                                                                        Widget:=texture_block{DefaultImage:=Textures.Boosters.T_IncreaseMaxShield, DefaultDesiredSize:=vector2{X:=25.0, Y:=25.0}, DefaultTint:=MakeColorFromHex("#FFFFFFFF")}

                                                                                                                    stack_box_slot:
                                                                                                                        VerticalAlignment:=vertical_alignment.Center
                                                                                                                        HorizontalAlignment:=horizontal_alignment.Center
                                                                                                                        #Distribution:=option{0.33}
                                                                                                                        Padding:=margin{Left:=0.0, Top:=0.0, Right:=20.0, Bottom:=5.0}
                                                                                                                        Widget:=text_block{DefaultText := StringToMessage("+50%"), DefaultTextColor:=MakeColorFromHex("#c3c3c3FF"), DefaultShadowColor:=MakeColorFromHex("#000000FF"), DefaultShadowOffset:=option{vector2{X:=3.0, Y:=3.0}}}
                                                                                                        
                                                                                                        #Base description text 2
                                                                                                        overlay_slot:
                                                                                                            VerticalAlignment:=vertical_alignment.Bottom
                                                                                                            HorizontalAlignment:=horizontal_alignment.Center
                                                                                                            #Distribution:=option{0.33}
                                                                                                            Padding:=margin{Left:=0.0, Top:=20.0, Right:=0.0, Bottom:=0.0}
                                                                                                            Widget:=stack_box:
                                                                                                                Orientation:=orientation.Horizontal
                                                                                                                Slots:=array:

                                                                                                                    stack_box_slot:
                                                                                                                        VerticalAlignment:=vertical_alignment.Center
                                                                                                                        HorizontalAlignment:=horizontal_alignment.Fill
                                                                                                                        Distribution:=option{1.0}
                                                                                                                        Padding:=margin{Left:=0.0, Top:=0.0, Right:=30.0, Bottom:=0.0}
                                                                                                                        Widget:=text_block{DefaultText := StringToMessage("once"), DefaultTextColor:=MakeColorFromHex("#c3c3c3FF"), DefaultShadowColor:=MakeColorFromHex("#000000FF"), DefaultShadowOffset:=option{vector2{X:=3.0, Y:=3.0}}}

                                                                                #Arrow
                                                                                stack_box_slot:
                                                                                    VerticalAlignment:=vertical_alignment.Center
                                                                                    HorizontalAlignment:=horizontal_alignment.Center
                                                                                    #Distribution:=option{0.33}
                                                                                    Padding:=margin{Left:=8.0, Top:=0.0, Right:=12.0, Bottom:=3.0}
                                                                                    Widget:=overlay:
                                                                                        Slots:=array:
                                                                                            
                                                                                            #Arrow
                                                                                            overlay_slot:
                                                                                                VerticalAlignment:=vertical_alignment.Center
                                                                                                HorizontalAlignment:=horizontal_alignment.Center
                                                                                                #Distribution:=option{0.33}
                                                                                                Padding:=margin{Left:=0.0, Top:=0.0, Right:=0.0, Bottom:=3.0}
                                                                                                Widget:=texture_block{DefaultImage:=Textures.General.T_GenArrowHead, DefaultDesiredSize:=vector2{X:=26.0, Y:=24.0}, DefaultTint:=MakeColorFromHex("#FFFFFFFF")}
                                                                                
                                                                                #Second description column
                                                                                stack_box_slot:
                                                                                    VerticalAlignment:=vertical_alignment.Center
                                                                                    HorizontalAlignment:=horizontal_alignment.Right
                                                                                    Distribution:=option{1.0}
                                                                                    Padding:=margin{Left:=0.0, Top:=0.0, Right:=0.0, Bottom:=0.0}
                                                                                    Widget:=stack_box:
                                                                                        Orientation:=orientation.Vertical
                                                                                        Slots:=array:

                                                                                            #First horizontal stack box
                                                                                            stack_box_slot:
                                                                                                VerticalAlignment:=vertical_alignment.Center
                                                                                                HorizontalAlignment:=horizontal_alignment.Center
                                                                                                Distribution:=option{1.0}
                                                                                                Padding:=margin{Left:=0.0, Top:=0.0, Right:=0.0, Bottom:=5.0}
                                                                                                Widget:=stack_box:
                                                                                                    Orientation:=orientation.Horizontal
                                                                                                    Slots:=array:

                                                                                                        #Lv 1 text
                                                                                                        stack_box_slot:
                                                                                                            VerticalAlignment:=vertical_alignment.Center
                                                                                                            HorizontalAlignment:=horizontal_alignment.Center
                                                                                                            #Distribution:=option{0.33}
                                                                                                            Padding:=margin{Left:=0.0, Top:=5.0, Right:=8.0, Bottom:=0.0}
                                                                                                            Widget:=text_block{DefaultText := StringToMessage("LV. 1"), DefaultTextColor:=MakeColorFromHex("#f6fe17FF"), DefaultShadowColor:=MakeColorFromHex("#000000FF"), DefaultShadowOffset:=option{vector2{X:=3.0, Y:=3.0}}}
                                                                                                        
                                                                                            #Second horizontal stack box (Actually an overlay)
                                                                                            stack_box_slot:
                                                                                                VerticalAlignment:=vertical_alignment.Center
                                                                                                HorizontalAlignment:=horizontal_alignment.Center
                                                                                                Distribution:=option{1.0}
                                                                                                Padding:=margin{Left:=0.0, Top:=0.0, Right:=0.0, Bottom:=0.0}
                                                                                                Widget:=overlay:
                                                                                                    Slots:=array:

                                                                                                        #Base description text 1
                                                                                                        overlay_slot:
                                                                                                            VerticalAlignment:=vertical_alignment.Center
                                                                                                            HorizontalAlignment:=horizontal_alignment.Center
                                                                                                            #Distribution:=option{0.33}
                                                                                                            Padding:=margin{Left:=0.0, Top:=0.0, Right:=0.0, Bottom:=0.0}
                                                                                                            Widget:=stack_box:
                                                                                                                Orientation:=orientation.Horizontal
                                                                                                                Slots:=array:

                                                                                                                    stack_box_slot:
                                                                                                                        VerticalAlignment:=vertical_alignment.Center
                                                                                                                        HorizontalAlignment:=horizontal_alignment.Center
                                                                                                                        #Distribution:=option{0.33}
                                                                                                                        Padding:=margin{Left:=0.0, Top:=0.0, Right:=7.0, Bottom:=5.0}
                                                                                                                        Widget:=texture_block{DefaultImage:=Textures.Boosters.T_IncreaseMaxShield, DefaultDesiredSize:=vector2{X:=25.0, Y:=25.0}, DefaultTint:=MakeColorFromHex("#FFFFFFFF")}

                                                                                                                    stack_box_slot:
                                                                                                                        VerticalAlignment:=vertical_alignment.Center
                                                                                                                        HorizontalAlignment:=horizontal_alignment.Center
                                                                                                                        #Distribution:=option{0.33}
                                                                                                                        Padding:=margin{Left:=0.0, Top:=0.0, Right:=20.0, Bottom:=5.0}
                                                                                                                        Widget:=text_block{DefaultText := StringToMessage("+50%"), DefaultTextColor:=MakeColorFromHex("#f6fe17FF"), DefaultShadowColor:=MakeColorFromHex("#000000FF"), DefaultShadowOffset:=option{vector2{X:=3.0, Y:=3.0}}}
                                                                                                                    
                                                                                                        #Base description text 2
                                                                                                        overlay_slot:
                                                                                                            VerticalAlignment:=vertical_alignment.Bottom
                                                                                                            HorizontalAlignment:=horizontal_alignment.Center
                                                                                                            #Distribution:=option{0.33}
                                                                                                            Padding:=margin{Left:=0.0, Top:=20.0, Right:=0.0, Bottom:=0.0}
                                                                                                            Widget:=stack_box:
                                                                                                                Orientation:=orientation.Horizontal
                                                                                                                Slots:=array:

                                                                                                                    stack_box_slot:
                                                                                                                        VerticalAlignment:=vertical_alignment.Center
                                                                                                                        HorizontalAlignment:=horizontal_alignment.Left
                                                                                                                        Distribution:=option{1.0}
                                                                                                                        Padding:=margin{Left:=0.0, Top:=0.0, Right:=30.0, Bottom:=0.0}
                                                                                                                        Widget:=text_block{DefaultText := StringToMessage("twice"), DefaultTextColor:=MakeColorFromHex("#f6fe17FF"), DefaultShadowColor:=MakeColorFromHex("#000000FF"), DefaultShadowOffset:=option{vector2{X:=3.0, Y:=3.0}}}
                                                                                                                    
                                                        #Spacing
                                                        stack_box_slot:
                                                            VerticalAlignment:=vertical_alignment.Center
                                                            HorizontalAlignment:=horizontal_alignment.Right
                                                            #Distribution:=option{1.0}
                                                            Padding:=margin{Left:=0.0, Top:=0.0, Right:=0.0, Bottom:=0.0}
                                                            Widget:=texture_block{DefaultImage:=Textures.General.T_GenSquTrans, DefaultDesiredSize:=vector2{X:=350.0, Y:=120.0}, DefaultTint:=MakeColorFromHex("#00FF00FF")}
                                                        
                                                        #Call to action
                                                        stack_box_slot:
                                                            VerticalAlignment:=vertical_alignment.Center
                                                            HorizontalAlignment:=horizontal_alignment.Right
                                                            #Distribution:=option{1.0}
                                                            Padding:=margin{Left:=0.0, Top:=0.0, Right:=0.0, Bottom:=0.0}
                                                            Widget:=overlay:
                                                                Slots:=array:

                                                                    #Overlay structure, first is the button itself
                                                                    overlay_slot:
                                                                        VerticalAlignment:=vertical_alignment.Center
                                                                        HorizontalAlignment:=horizontal_alignment.Center
                                                                        #Distribution:=option{1.0}
                                                                        Padding:=margin{Left:=0.0, Top:=0.0, Right:=0.0, Bottom:=0.0}
                                                                        Widget:=button_loud{DefaultText:=StringToMessage("UPGRADE\n ")}

                                                                    #Overlay structure, icon and value
                                                                    overlay_slot:
                                                                        VerticalAlignment:=vertical_alignment.Fill
                                                                        HorizontalAlignment:=horizontal_alignment.Center
                                                                        #Distribution:=option{1.0}
                                                                        Padding:=margin{Left:=0.0, Top:=0.0, Right:=0.0, Bottom:=0.0}
                                                                        Widget:=stack_box:
                                                                            Orientation:=orientation.Horizontal
                                                                            Slots:=array:

                                                                                #Icon and value
                                                                                stack_box_slot:
                                                                                    VerticalAlignment:=vertical_alignment.Bottom
                                                                                    HorizontalAlignment:=horizontal_alignment.Right
                                                                                    #Distribution:=option{1.0}
                                                                                    Padding:=margin{Left:=0.0, Top:=0.0, Right:=0.0, Bottom:=8.0}
                                                                                    Widget:=texture_block{DefaultImage:=Textures.Currency.T_Gold, DefaultDesiredSize:=vector2{X:=35.0, Y:=35.0}, DefaultTint:=MakeColorFromHex("#000000FF")}                

                                                                                stack_box_slot:
                                                                                    VerticalAlignment:=vertical_alignment.Bottom
                                                                                    HorizontalAlignment:=horizontal_alignment.Center
                                                                                    #Distribution:=option{1.0}
                                                                                    Padding:=margin{Left:=5.0, Top:=0.0, Right:=0.0, Bottom:=5.0}
                                                                                    Widget:=text_block{DefaultText := StringToMessage("15"), DefaultTextColor:=MakeColorFromHex("#000000FF"), DefaultShadowColor:=MakeColorFromHex("#000000FF"), DefaultShadowOffset:=option{vector2{X:=3.0, Y:=3.0}}}

            NavPurchase3:=overlay:
                Slots:=array:

                    #Makeshift size box value
                    overlay_slot:
                        VerticalAlignment:=vertical_alignment.Center
                        HorizontalAlignment:=horizontal_alignment.Center
                        #Distribution:=option{0.33}
                        Padding:=margin{Left:=0.0, Top:=0.0, Right:=0.0, Bottom:=0.0}
                        Widget:=texture_block{DefaultImage:=Textures.General.T_GenSquTrans, DefaultDesiredSize:=vector2{X:=1800.0, Y:=760.0}, DefaultTint:=MakeColorFromHex("#00GG00FF")}

                    #Main content structure
                    overlay_slot:
                        VerticalAlignment:=vertical_alignment.Center
                        HorizontalAlignment:=horizontal_alignment.Center
                        #Distribution:=option{0.33}
                        Padding:=margin{Left:=0.0, Top:=0.0, Right:=0.0, Bottom:=0.0}
                        Widget:=stack_box:
                            Orientation:=orientation.Vertical
                            Slots:=array:

                                #Single row layout
                                stack_box_slot:
                                    VerticalAlignment:=vertical_alignment.Center
                                    HorizontalAlignment:=horizontal_alignment.Center
                                    #Distribution:=option{0.33}
                                    Padding:=margin{Left:=0.0, Top:=0.0, Right:=0.0, Bottom:=0.0}
                                    Widget:=overlay:
                                        Slots:=array:
                                            
                                            #Makeshift size box
                                                overlay_slot:
                                                    VerticalAlignment:=vertical_alignment.Center
                                                    HorizontalAlignment:=horizontal_alignment.Center
                                                    #Distribution:=option{0.33}
                                                    Padding:=margin{Left:=0.0, Top:=0.0, Right:=0.0, Bottom:=0.0}
                                                    Widget:=texture_block{DefaultImage:=Textures.General.T_GenSquTrans, DefaultDesiredSize:=vector2{X:=1600.0, Y:=120.0}, DefaultTint:=MakeColorFromHex("#FF0000FF")}        

                                            #Row layout
                                            overlay_slot:
                                                VerticalAlignment:=vertical_alignment.Center
                                                HorizontalAlignment:=horizontal_alignment.Fill
                                                #Distribution:=option{0.33}
                                                Padding:=margin{Left:=0.0, Top:=0.0, Right:=0.0, Bottom:=0.0}
                                                Widget:=stack_box:
                                                    Orientation:=orientation.Horizontal
                                                    Slots:=array:
                                                        
                                                        #Booster icon and rarity
                                                        stack_box_slot:
                                                            VerticalAlignment:=vertical_alignment.Center
                                                            HorizontalAlignment:=horizontal_alignment.Left
                                                            Distribution:=option{1.0}
                                                            Padding:=margin{Left:=0.0, Top:=0.0, Right:=0.0, Bottom:=0.0}
                                                            Widget:=overlay:
                                                                Slots:=array:
                                                                    
                                                                    #Booster Rarity BG
                                                                    overlay_slot:
                                                                        VerticalAlignment:=vertical_alignment.Center
                                                                        HorizontalAlignment:=horizontal_alignment.Left
                                                                        #Distribution:=option{1.0}
                                                                        Padding:=margin{Left:=0.0, Top:=0.0, Right:=0.0, Bottom:=0.0}
                                                                        Widget:=texture_block{DefaultImage:=Textures.Rarity.T_IconBG_Common, DefaultDesiredSize:=vector2{X:=120.0, Y:=120.0}, DefaultTint:=MakeColorFromHex("#FFFFFFFF")}
                                                                    
                                                                    #Booster Icon
                                                                    overlay_slot:
                                                                        VerticalAlignment:=vertical_alignment.Center
                                                                        HorizontalAlignment:=horizontal_alignment.Left
                                                                        #Distribution:=option{1.0}
                                                                        Padding:=margin{Left:=0.0, Top:=0.0, Right:=0.0, Bottom:=0.0}
                                                                        Widget:=texture_block{DefaultImage:=Textures.Boosters.T_IncreaseMaxShield, DefaultDesiredSize:=vector2{X:=116.0, Y:=116.0}, DefaultTint:=MakeColorFromHex("#FFFFFFFF")}

                                                                    #Booster Border
                                                                    overlay_slot:
                                                                        VerticalAlignment:=vertical_alignment.Center
                                                                        HorizontalAlignment:=horizontal_alignment.Left
                                                                        #Distribution:=option{1.0}
                                                                        Padding:=margin{Left:=0.0, Top:=0.0, Right:=0.0, Bottom:=0.0}
                                                                        Widget:=texture_block{DefaultImage:=Textures.General.T_GenericBorderSquare, DefaultDesiredSize:=vector2{X:=120.0, Y:=120.0}, DefaultTint:=MakeColorFromHex("#cdc953FF")}
                                                                                                                                                                                                        
                                                        #Booster description
                                                        stack_box_slot:
                                                            VerticalAlignment:=vertical_alignment.Center
                                                            HorizontalAlignment:=horizontal_alignment.Fill
                                                            #Distribution:=option{1.0}
                                                            Padding:=margin{Left:=0.0, Top:=0.0, Right:=0.0, Bottom:=0.0}
                                                            Widget:=overlay:
                                                                Slots:=array:

                                                                    #BG color
                                                                    overlay_slot:
                                                                        VerticalAlignment:=vertical_alignment.Center
                                                                        HorizontalAlignment:=horizontal_alignment.Center
                                                                        #Distribution:=option{0.33}
                                                                        Padding:=margin{Left:=0.0, Top:=0.0, Right:=0.0, Bottom:=0.0}
                                                                        Widget:=texture_block{DefaultImage:=Textures.General.T_GenericSquare, DefaultDesiredSize:=vector2{X:=570.0, Y:=120.0}, DefaultTint:=MakeColorFromHex("#031126FF")}

                                                                    #Overall Border
                                                                    overlay_slot:
                                                                        VerticalAlignment:=vertical_alignment.Center
                                                                        HorizontalAlignment:=horizontal_alignment.Center
                                                                        #Distribution:=option{0.33}
                                                                        Padding:=margin{Left:=0.0, Top:=0.0, Right:=0.0, Bottom:=0.0}
                                                                        Widget:=texture_block{DefaultImage:=Textures.General.T_LevelBorder, DefaultDesiredSize:=vector2{X:=570.0, Y:=120.0}, DefaultTint:=MakeColorFromHex("#1d4b89FF")}
                                                                    
                                                                        
                                                                        
                                                                    #Description Header
                                                                    overlay_slot:
                                                                        VerticalAlignment:=vertical_alignment.Top
                                                                        HorizontalAlignment:=horizontal_alignment.Center
                                                                        #Distribution:=option{0.33}
                                                                        Padding:=margin{Left:=0.0, Top:=0.0, Right:=0.0, Bottom:=5.0}
                                                                        Widget:=text_block{DefaultText := StringToMessage("Increase Max Shield"), DefaultTextColor:=MakeColorFromHex("#f6fe17FF"), DefaultShadowColor:=MakeColorFromHex("#000000FF"), DefaultShadowOffset:=option{vector2{X:=3.0, Y:=3.0}}}

                                                                    #Description Header
                                                                    overlay_slot:
                                                                        VerticalAlignment:=vertical_alignment.Bottom
                                                                        HorizontalAlignment:=horizontal_alignment.Center
                                                                        #Distribution:=option{0.33}
                                                                        Padding:=margin{Left:=0.0, Top:=0.0, Right:=0.0, Bottom:=0.0}
                                                                        Widget:=text_block{DefaultText := StringToMessage("If health falls under 40%,\nrecover 50% shield once"), DefaultTextColor:=MakeColorFromHex("#c4c5c7FF"), DefaultShadowColor:=MakeColorFromHex("#000000FF"), DefaultShadowOffset:=option{vector2{X:=3.0, Y:=3.0}}}

                                                        #Upgrade description
                                                        stack_box_slot:
                                                            VerticalAlignment:=vertical_alignment.Center
                                                            HorizontalAlignment:=horizontal_alignment.Center
                                                            #Distribution:=option{0.33}
                                                            Padding:=margin{Left:=0.0, Top:=0.0, Right:=0.0, Bottom:=0.0}
                                                            Widget:=overlay:
                                                                Slots:=array:

                                                                    #BG color
                                                                    overlay_slot:
                                                                        VerticalAlignment:=vertical_alignment.Center
                                                                        HorizontalAlignment:=horizontal_alignment.Center
                                                                        #Distribution:=option{0.33}
                                                                        Padding:=margin{Left:=0.0, Top:=0.0, Right:=0.0, Bottom:=0.0}
                                                                        Widget:=texture_block{DefaultImage:=Textures.General.T_GenericSquare, DefaultDesiredSize:=vector2{X:=348.0, Y:=120.0}, DefaultTint:=MakeColorFromHex("#000000FF")}

                                                                    #Overall Border
                                                                    overlay_slot:
                                                                        VerticalAlignment:=vertical_alignment.Center
                                                                        HorizontalAlignment:=horizontal_alignment.Center
                                                                        #Distribution:=option{0.33}
                                                                        Padding:=margin{Left:=0.0, Top:=0.0, Right:=0.0, Bottom:=0.0}
                                                                        Widget:=texture_block{DefaultImage:=Textures.General.T_LevelBorder, DefaultDesiredSize:=vector2{X:=348.0, Y:=120.0}, DefaultTint:=MakeColorFromHex("#1d4b89FF")}

                                                                    #Upgrade description layout - Horizontal stack box
                                                                    overlay_slot:
                                                                        VerticalAlignment:=vertical_alignment.Center
                                                                        HorizontalAlignment:=horizontal_alignment.Fill
                                                                        #Distribution:=option{1.0}
                                                                        Padding:=margin{Left:=0.0, Top:=0.0, Right:=0.0, Bottom:=5.0}
                                                                        Widget:=stack_box:
                                                                            Orientation:=orientation.Horizontal
                                                                            Slots:=array:

                                                                                #First description column
                                                                                stack_box_slot:
                                                                                    VerticalAlignment:=vertical_alignment.Center
                                                                                    HorizontalAlignment:=horizontal_alignment.Left
                                                                                    Distribution:=option{1.0}
                                                                                    Padding:=margin{Left:=25.0, Top:=0.0, Right:=0.0, Bottom:=0.0}
                                                                                    Widget:=stack_box:
                                                                                        Orientation:=orientation.Vertical
                                                                                        Slots:=array:

                                                                                            #First horizontal stack box
                                                                                            stack_box_slot:
                                                                                                VerticalAlignment:=vertical_alignment.Center
                                                                                                HorizontalAlignment:=horizontal_alignment.Center
                                                                                                Distribution:=option{1.0}
                                                                                                Padding:=margin{Left:=0.0, Top:=0.0, Right:=0.0, Bottom:=5.0}
                                                                                                Widget:=stack_box:
                                                                                                    Orientation:=orientation.Horizontal
                                                                                                    Slots:=array:

                                                                                                        #Base Lv text
                                                                                                        stack_box_slot:
                                                                                                            VerticalAlignment:=vertical_alignment.Center
                                                                                                            HorizontalAlignment:=horizontal_alignment.Center
                                                                                                            #Distribution:=option{0.33}
                                                                                                            Padding:=margin{Left:=0.0, Top:=5.0, Right:=8.0, Bottom:=0.0}
                                                                                                            Widget:=text_block{DefaultText := StringToMessage("BASE LV"), DefaultTextColor:=MakeColorFromHex("#FFFFFFFF"), DefaultShadowColor:=MakeColorFromHex("#000000FF"), DefaultShadowOffset:=option{vector2{X:=3.0, Y:=3.0}}}

                                                                                            #Second horizontal stack box (Actually an overlay)
                                                                                            stack_box_slot:
                                                                                                VerticalAlignment:=vertical_alignment.Center
                                                                                                HorizontalAlignment:=horizontal_alignment.Center
                                                                                                Distribution:=option{1.0}
                                                                                                Padding:=margin{Left:=0.0, Top:=0.0, Right:=0.0, Bottom:=0.0}
                                                                                                Widget:=overlay:
                                                                                                    Slots:=array:

                                                                                                        #Base description text 1
                                                                                                        overlay_slot:
                                                                                                            VerticalAlignment:=vertical_alignment.Center
                                                                                                            HorizontalAlignment:=horizontal_alignment.Center
                                                                                                            #Distribution:=option{0.33}
                                                                                                            Padding:=margin{Left:=0.0, Top:=0.0, Right:=0.0, Bottom:=0.0}
                                                                                                            Widget:=stack_box:
                                                                                                                Orientation:=orientation.Horizontal
                                                                                                                Slots:=array:

                                                                                                                    stack_box_slot:
                                                                                                                        VerticalAlignment:=vertical_alignment.Center
                                                                                                                        HorizontalAlignment:=horizontal_alignment.Center
                                                                                                                        #Distribution:=option{0.33}
                                                                                                                        Padding:=margin{Left:=0.0, Top:=0.0, Right:=7.0, Bottom:=5.0}
                                                                                                                        Widget:=texture_block{DefaultImage:=Textures.Boosters.T_IncreaseMaxShield, DefaultDesiredSize:=vector2{X:=25.0, Y:=25.0}, DefaultTint:=MakeColorFromHex("#FFFFFFFF")}

                                                                                                                    stack_box_slot:
                                                                                                                        VerticalAlignment:=vertical_alignment.Center
                                                                                                                        HorizontalAlignment:=horizontal_alignment.Center
                                                                                                                        #Distribution:=option{0.33}
                                                                                                                        Padding:=margin{Left:=0.0, Top:=0.0, Right:=20.0, Bottom:=5.0}
                                                                                                                        Widget:=text_block{DefaultText := StringToMessage("+50%"), DefaultTextColor:=MakeColorFromHex("#c3c3c3FF"), DefaultShadowColor:=MakeColorFromHex("#000000FF"), DefaultShadowOffset:=option{vector2{X:=3.0, Y:=3.0}}}
                                                                                                        
                                                                                                        #Base description text 2
                                                                                                        overlay_slot:
                                                                                                            VerticalAlignment:=vertical_alignment.Bottom
                                                                                                            HorizontalAlignment:=horizontal_alignment.Center
                                                                                                            #Distribution:=option{0.33}
                                                                                                            Padding:=margin{Left:=0.0, Top:=20.0, Right:=0.0, Bottom:=0.0}
                                                                                                            Widget:=stack_box:
                                                                                                                Orientation:=orientation.Horizontal
                                                                                                                Slots:=array:

                                                                                                                    stack_box_slot:
                                                                                                                        VerticalAlignment:=vertical_alignment.Center
                                                                                                                        HorizontalAlignment:=horizontal_alignment.Fill
                                                                                                                        Distribution:=option{1.0}
                                                                                                                        Padding:=margin{Left:=0.0, Top:=0.0, Right:=30.0, Bottom:=0.0}
                                                                                                                        Widget:=text_block{DefaultText := StringToMessage("once"), DefaultTextColor:=MakeColorFromHex("#c3c3c3FF"), DefaultShadowColor:=MakeColorFromHex("#000000FF"), DefaultShadowOffset:=option{vector2{X:=3.0, Y:=3.0}}}

                                                                                #Arrow
                                                                                stack_box_slot:
                                                                                    VerticalAlignment:=vertical_alignment.Center
                                                                                    HorizontalAlignment:=horizontal_alignment.Center
                                                                                    #Distribution:=option{0.33}
                                                                                    Padding:=margin{Left:=8.0, Top:=0.0, Right:=12.0, Bottom:=3.0}
                                                                                    Widget:=overlay:
                                                                                        Slots:=array:
                                                                                            
                                                                                            #Arrow
                                                                                            overlay_slot:
                                                                                                VerticalAlignment:=vertical_alignment.Center
                                                                                                HorizontalAlignment:=horizontal_alignment.Center
                                                                                                #Distribution:=option{0.33}
                                                                                                Padding:=margin{Left:=0.0, Top:=0.0, Right:=0.0, Bottom:=3.0}
                                                                                                Widget:=texture_block{DefaultImage:=Textures.General.T_GenArrowHead, DefaultDesiredSize:=vector2{X:=26.0, Y:=24.0}, DefaultTint:=MakeColorFromHex("#FFFFFFFF")}
                                                                                
                                                                                #Second description column
                                                                                stack_box_slot:
                                                                                    VerticalAlignment:=vertical_alignment.Center
                                                                                    HorizontalAlignment:=horizontal_alignment.Right
                                                                                    Distribution:=option{1.0}
                                                                                    Padding:=margin{Left:=0.0, Top:=0.0, Right:=0.0, Bottom:=0.0}
                                                                                    Widget:=stack_box:
                                                                                        Orientation:=orientation.Vertical
                                                                                        Slots:=array:

                                                                                            #First horizontal stack box
                                                                                            stack_box_slot:
                                                                                                VerticalAlignment:=vertical_alignment.Center
                                                                                                HorizontalAlignment:=horizontal_alignment.Center
                                                                                                Distribution:=option{1.0}
                                                                                                Padding:=margin{Left:=0.0, Top:=0.0, Right:=0.0, Bottom:=5.0}
                                                                                                Widget:=stack_box:
                                                                                                    Orientation:=orientation.Horizontal
                                                                                                    Slots:=array:

                                                                                                        #Lv 1 text
                                                                                                        stack_box_slot:
                                                                                                            VerticalAlignment:=vertical_alignment.Center
                                                                                                            HorizontalAlignment:=horizontal_alignment.Center
                                                                                                            #Distribution:=option{0.33}
                                                                                                            Padding:=margin{Left:=0.0, Top:=5.0, Right:=8.0, Bottom:=0.0}
                                                                                                            Widget:=text_block{DefaultText := StringToMessage("LV. 1"), DefaultTextColor:=MakeColorFromHex("#f6fe17FF"), DefaultShadowColor:=MakeColorFromHex("#000000FF"), DefaultShadowOffset:=option{vector2{X:=3.0, Y:=3.0}}}
                                                                                                        
                                                                                            #Second horizontal stack box (Actually an overlay)
                                                                                            stack_box_slot:
                                                                                                VerticalAlignment:=vertical_alignment.Center
                                                                                                HorizontalAlignment:=horizontal_alignment.Center
                                                                                                Distribution:=option{1.0}
                                                                                                Padding:=margin{Left:=0.0, Top:=0.0, Right:=0.0, Bottom:=0.0}
                                                                                                Widget:=overlay:
                                                                                                    Slots:=array:

                                                                                                        #Base description text 1
                                                                                                        overlay_slot:
                                                                                                            VerticalAlignment:=vertical_alignment.Center
                                                                                                            HorizontalAlignment:=horizontal_alignment.Center
                                                                                                            #Distribution:=option{0.33}
                                                                                                            Padding:=margin{Left:=0.0, Top:=0.0, Right:=0.0, Bottom:=0.0}
                                                                                                            Widget:=stack_box:
                                                                                                                Orientation:=orientation.Horizontal
                                                                                                                Slots:=array:

                                                                                                                    stack_box_slot:
                                                                                                                        VerticalAlignment:=vertical_alignment.Center
                                                                                                                        HorizontalAlignment:=horizontal_alignment.Center
                                                                                                                        #Distribution:=option{0.33}
                                                                                                                        Padding:=margin{Left:=0.0, Top:=0.0, Right:=7.0, Bottom:=5.0}
                                                                                                                        Widget:=texture_block{DefaultImage:=Textures.Boosters.T_IncreaseMaxShield, DefaultDesiredSize:=vector2{X:=25.0, Y:=25.0}, DefaultTint:=MakeColorFromHex("#FFFFFFFF")}

                                                                                                                    stack_box_slot:
                                                                                                                        VerticalAlignment:=vertical_alignment.Center
                                                                                                                        HorizontalAlignment:=horizontal_alignment.Center
                                                                                                                        #Distribution:=option{0.33}
                                                                                                                        Padding:=margin{Left:=0.0, Top:=0.0, Right:=20.0, Bottom:=5.0}
                                                                                                                        Widget:=text_block{DefaultText := StringToMessage("+50%"), DefaultTextColor:=MakeColorFromHex("#f6fe17FF"), DefaultShadowColor:=MakeColorFromHex("#000000FF"), DefaultShadowOffset:=option{vector2{X:=3.0, Y:=3.0}}}
                                                                                                                    
                                                                                                        #Base description text 2
                                                                                                        overlay_slot:
                                                                                                            VerticalAlignment:=vertical_alignment.Bottom
                                                                                                            HorizontalAlignment:=horizontal_alignment.Center
                                                                                                            #Distribution:=option{0.33}
                                                                                                            Padding:=margin{Left:=0.0, Top:=20.0, Right:=0.0, Bottom:=0.0}
                                                                                                            Widget:=stack_box:
                                                                                                                Orientation:=orientation.Horizontal
                                                                                                                Slots:=array:

                                                                                                                    stack_box_slot:
                                                                                                                        VerticalAlignment:=vertical_alignment.Center
                                                                                                                        HorizontalAlignment:=horizontal_alignment.Left
                                                                                                                        Distribution:=option{1.0}
                                                                                                                        Padding:=margin{Left:=0.0, Top:=0.0, Right:=30.0, Bottom:=0.0}
                                                                                                                        Widget:=text_block{DefaultText := StringToMessage("twice"), DefaultTextColor:=MakeColorFromHex("#f6fe17FF"), DefaultShadowColor:=MakeColorFromHex("#000000FF"), DefaultShadowOffset:=option{vector2{X:=3.0, Y:=3.0}}}
                                                                                                                    
                                                        #Spacing
                                                        stack_box_slot:
                                                            VerticalAlignment:=vertical_alignment.Center
                                                            HorizontalAlignment:=horizontal_alignment.Right
                                                            #Distribution:=option{1.0}
                                                            Padding:=margin{Left:=0.0, Top:=0.0, Right:=0.0, Bottom:=0.0}
                                                            Widget:=texture_block{DefaultImage:=Textures.General.T_GenSquTrans, DefaultDesiredSize:=vector2{X:=350.0, Y:=120.0}, DefaultTint:=MakeColorFromHex("#00FF00FF")}
                                                        
                                                        #Call to action
                                                        stack_box_slot:
                                                            VerticalAlignment:=vertical_alignment.Center
                                                            HorizontalAlignment:=horizontal_alignment.Right
                                                            #Distribution:=option{1.0}
                                                            Padding:=margin{Left:=0.0, Top:=0.0, Right:=0.0, Bottom:=0.0}
                                                            Widget:=overlay:
                                                                Slots:=array:

                                                                    #Overlay structure, first is the button itself
                                                                    overlay_slot:
                                                                        VerticalAlignment:=vertical_alignment.Center
                                                                        HorizontalAlignment:=horizontal_alignment.Center
                                                                        #Distribution:=option{1.0}
                                                                        Padding:=margin{Left:=0.0, Top:=0.0, Right:=0.0, Bottom:=0.0}
                                                                        Widget:=button_loud{DefaultText:=StringToMessage("UPGRADE\n ")}

                                                                    #Overlay structure, icon and value
                                                                    overlay_slot:
                                                                        VerticalAlignment:=vertical_alignment.Fill
                                                                        HorizontalAlignment:=horizontal_alignment.Center
                                                                        #Distribution:=option{1.0}
                                                                        Padding:=margin{Left:=0.0, Top:=0.0, Right:=0.0, Bottom:=0.0}
                                                                        Widget:=stack_box:
                                                                            Orientation:=orientation.Horizontal
                                                                            Slots:=array:

                                                                                #Icon and value
                                                                                stack_box_slot:
                                                                                    VerticalAlignment:=vertical_alignment.Bottom
                                                                                    HorizontalAlignment:=horizontal_alignment.Right
                                                                                    #Distribution:=option{1.0}
                                                                                    Padding:=margin{Left:=0.0, Top:=0.0, Right:=0.0, Bottom:=8.0}
                                                                                    Widget:=texture_block{DefaultImage:=Textures.Currency.T_Gold, DefaultDesiredSize:=vector2{X:=35.0, Y:=35.0}, DefaultTint:=MakeColorFromHex("#000000FF")}                

                                                                                stack_box_slot:
                                                                                    VerticalAlignment:=vertical_alignment.Bottom
                                                                                    HorizontalAlignment:=horizontal_alignment.Center
                                                                                    #Distribution:=option{1.0}
                                                                                    Padding:=margin{Left:=5.0, Top:=0.0, Right:=0.0, Bottom:=5.0}
                                                                                    Widget:=text_block{DefaultText := StringToMessage("15"), DefaultTextColor:=MakeColorFromHex("#000000FF"), DefaultShadowColor:=MakeColorFromHex("#000000FF"), DefaultShadowOffset:=option{vector2{X:=3.0, Y:=3.0}}}

            NavPurchase4:=overlay:
                Slots:=array:

                    #Makeshift size box value
                    overlay_slot:
                        VerticalAlignment:=vertical_alignment.Center
                        HorizontalAlignment:=horizontal_alignment.Center
                        #Distribution:=option{0.33}
                        Padding:=margin{Left:=0.0, Top:=0.0, Right:=0.0, Bottom:=0.0}
                        Widget:=texture_block{DefaultImage:=Textures.General.T_GenSquTrans, DefaultDesiredSize:=vector2{X:=1800.0, Y:=760.0}, DefaultTint:=MakeColorFromHex("#00GG00FF")}

                    #Main content structure
                    overlay_slot:
                        VerticalAlignment:=vertical_alignment.Center
                        HorizontalAlignment:=horizontal_alignment.Center
                        #Distribution:=option{0.33}
                        Padding:=margin{Left:=0.0, Top:=0.0, Right:=0.0, Bottom:=0.0}
                        Widget:=stack_box:
                            Orientation:=orientation.Vertical
                            Slots:=array:

                                #Single row layout
                                stack_box_slot:
                                    VerticalAlignment:=vertical_alignment.Center
                                    HorizontalAlignment:=horizontal_alignment.Center
                                    #Distribution:=option{0.33}
                                    Padding:=margin{Left:=0.0, Top:=0.0, Right:=0.0, Bottom:=0.0}
                                    Widget:=overlay:
                                        Slots:=array:
                                            
                                            #Makeshift size box
                                                overlay_slot:
                                                    VerticalAlignment:=vertical_alignment.Center
                                                    HorizontalAlignment:=horizontal_alignment.Center
                                                    #Distribution:=option{0.33}
                                                    Padding:=margin{Left:=0.0, Top:=0.0, Right:=0.0, Bottom:=0.0}
                                                    Widget:=texture_block{DefaultImage:=Textures.General.T_GenSquTrans, DefaultDesiredSize:=vector2{X:=1600.0, Y:=120.0}, DefaultTint:=MakeColorFromHex("#FF0000FF")}        

                                            #Row layout
                                            overlay_slot:
                                                VerticalAlignment:=vertical_alignment.Center
                                                HorizontalAlignment:=horizontal_alignment.Fill
                                                #Distribution:=option{0.33}
                                                Padding:=margin{Left:=0.0, Top:=0.0, Right:=0.0, Bottom:=0.0}
                                                Widget:=stack_box:
                                                    Orientation:=orientation.Horizontal
                                                    Slots:=array:
                                                        
                                                        #Booster icon and rarity
                                                        stack_box_slot:
                                                            VerticalAlignment:=vertical_alignment.Center
                                                            HorizontalAlignment:=horizontal_alignment.Left
                                                            Distribution:=option{1.0}
                                                            Padding:=margin{Left:=0.0, Top:=0.0, Right:=0.0, Bottom:=0.0}
                                                            Widget:=overlay:
                                                                Slots:=array:
                                                                    
                                                                    #Booster Rarity BG
                                                                    overlay_slot:
                                                                        VerticalAlignment:=vertical_alignment.Center
                                                                        HorizontalAlignment:=horizontal_alignment.Left
                                                                        #Distribution:=option{1.0}
                                                                        Padding:=margin{Left:=0.0, Top:=0.0, Right:=0.0, Bottom:=0.0}
                                                                        Widget:=texture_block{DefaultImage:=Textures.Rarity.T_IconBG_Common, DefaultDesiredSize:=vector2{X:=120.0, Y:=120.0}, DefaultTint:=MakeColorFromHex("#FFFFFFFF")}
                                                                    
                                                                    #Booster Icon
                                                                    overlay_slot:
                                                                        VerticalAlignment:=vertical_alignment.Center
                                                                        HorizontalAlignment:=horizontal_alignment.Left
                                                                        #Distribution:=option{1.0}
                                                                        Padding:=margin{Left:=0.0, Top:=0.0, Right:=0.0, Bottom:=0.0}
                                                                        Widget:=texture_block{DefaultImage:=Textures.Boosters.T_IncreaseMaxShield, DefaultDesiredSize:=vector2{X:=116.0, Y:=116.0}, DefaultTint:=MakeColorFromHex("#FFFFFFFF")}

                                                                    #Booster Border
                                                                    overlay_slot:
                                                                        VerticalAlignment:=vertical_alignment.Center
                                                                        HorizontalAlignment:=horizontal_alignment.Left
                                                                        #Distribution:=option{1.0}
                                                                        Padding:=margin{Left:=0.0, Top:=0.0, Right:=0.0, Bottom:=0.0}
                                                                        Widget:=texture_block{DefaultImage:=Textures.General.T_GenericBorderSquare, DefaultDesiredSize:=vector2{X:=120.0, Y:=120.0}, DefaultTint:=MakeColorFromHex("#cdc953FF")}
                                                                                                                                                                                                        
                                                        #Booster description
                                                        stack_box_slot:
                                                            VerticalAlignment:=vertical_alignment.Center
                                                            HorizontalAlignment:=horizontal_alignment.Fill
                                                            #Distribution:=option{1.0}
                                                            Padding:=margin{Left:=0.0, Top:=0.0, Right:=0.0, Bottom:=0.0}
                                                            Widget:=overlay:
                                                                Slots:=array:

                                                                    #BG color
                                                                    overlay_slot:
                                                                        VerticalAlignment:=vertical_alignment.Center
                                                                        HorizontalAlignment:=horizontal_alignment.Center
                                                                        #Distribution:=option{0.33}
                                                                        Padding:=margin{Left:=0.0, Top:=0.0, Right:=0.0, Bottom:=0.0}
                                                                        Widget:=texture_block{DefaultImage:=Textures.General.T_GenericSquare, DefaultDesiredSize:=vector2{X:=570.0, Y:=120.0}, DefaultTint:=MakeColorFromHex("#031126FF")}

                                                                    #Overall Border
                                                                    overlay_slot:
                                                                        VerticalAlignment:=vertical_alignment.Center
                                                                        HorizontalAlignment:=horizontal_alignment.Center
                                                                        #Distribution:=option{0.33}
                                                                        Padding:=margin{Left:=0.0, Top:=0.0, Right:=0.0, Bottom:=0.0}
                                                                        Widget:=texture_block{DefaultImage:=Textures.General.T_LevelBorder, DefaultDesiredSize:=vector2{X:=570.0, Y:=120.0}, DefaultTint:=MakeColorFromHex("#1d4b89FF")}
                                                                    
                                                                        
                                                                        
                                                                    #Description Header
                                                                    overlay_slot:
                                                                        VerticalAlignment:=vertical_alignment.Top
                                                                        HorizontalAlignment:=horizontal_alignment.Center
                                                                        #Distribution:=option{0.33}
                                                                        Padding:=margin{Left:=0.0, Top:=0.0, Right:=0.0, Bottom:=5.0}
                                                                        Widget:=text_block{DefaultText := StringToMessage("Increase Max Shield"), DefaultTextColor:=MakeColorFromHex("#f6fe17FF"), DefaultShadowColor:=MakeColorFromHex("#000000FF"), DefaultShadowOffset:=option{vector2{X:=3.0, Y:=3.0}}}

                                                                    #Description Header
                                                                    overlay_slot:
                                                                        VerticalAlignment:=vertical_alignment.Bottom
                                                                        HorizontalAlignment:=horizontal_alignment.Center
                                                                        #Distribution:=option{0.33}
                                                                        Padding:=margin{Left:=0.0, Top:=0.0, Right:=0.0, Bottom:=0.0}
                                                                        Widget:=text_block{DefaultText := StringToMessage("If health falls under 40%,\nrecover 50% shield once"), DefaultTextColor:=MakeColorFromHex("#c4c5c7FF"), DefaultShadowColor:=MakeColorFromHex("#000000FF"), DefaultShadowOffset:=option{vector2{X:=3.0, Y:=3.0}}}

                                                        #Upgrade description
                                                        stack_box_slot:
                                                            VerticalAlignment:=vertical_alignment.Center
                                                            HorizontalAlignment:=horizontal_alignment.Center
                                                            #Distribution:=option{0.33}
                                                            Padding:=margin{Left:=0.0, Top:=0.0, Right:=0.0, Bottom:=0.0}
                                                            Widget:=overlay:
                                                                Slots:=array:

                                                                    #BG color
                                                                    overlay_slot:
                                                                        VerticalAlignment:=vertical_alignment.Center
                                                                        HorizontalAlignment:=horizontal_alignment.Center
                                                                        #Distribution:=option{0.33}
                                                                        Padding:=margin{Left:=0.0, Top:=0.0, Right:=0.0, Bottom:=0.0}
                                                                        Widget:=texture_block{DefaultImage:=Textures.General.T_GenericSquare, DefaultDesiredSize:=vector2{X:=348.0, Y:=120.0}, DefaultTint:=MakeColorFromHex("#000000FF")}

                                                                    #Overall Border
                                                                    overlay_slot:
                                                                        VerticalAlignment:=vertical_alignment.Center
                                                                        HorizontalAlignment:=horizontal_alignment.Center
                                                                        #Distribution:=option{0.33}
                                                                        Padding:=margin{Left:=0.0, Top:=0.0, Right:=0.0, Bottom:=0.0}
                                                                        Widget:=texture_block{DefaultImage:=Textures.General.T_LevelBorder, DefaultDesiredSize:=vector2{X:=348.0, Y:=120.0}, DefaultTint:=MakeColorFromHex("#1d4b89FF")}

                                                                    #Upgrade description layout - Horizontal stack box
                                                                    overlay_slot:
                                                                        VerticalAlignment:=vertical_alignment.Center
                                                                        HorizontalAlignment:=horizontal_alignment.Fill
                                                                        #Distribution:=option{1.0}
                                                                        Padding:=margin{Left:=0.0, Top:=0.0, Right:=0.0, Bottom:=5.0}
                                                                        Widget:=stack_box:
                                                                            Orientation:=orientation.Horizontal
                                                                            Slots:=array:

                                                                                #First description column
                                                                                stack_box_slot:
                                                                                    VerticalAlignment:=vertical_alignment.Center
                                                                                    HorizontalAlignment:=horizontal_alignment.Left
                                                                                    Distribution:=option{1.0}
                                                                                    Padding:=margin{Left:=25.0, Top:=0.0, Right:=0.0, Bottom:=0.0}
                                                                                    Widget:=stack_box:
                                                                                        Orientation:=orientation.Vertical
                                                                                        Slots:=array:

                                                                                            #First horizontal stack box
                                                                                            stack_box_slot:
                                                                                                VerticalAlignment:=vertical_alignment.Center
                                                                                                HorizontalAlignment:=horizontal_alignment.Center
                                                                                                Distribution:=option{1.0}
                                                                                                Padding:=margin{Left:=0.0, Top:=0.0, Right:=0.0, Bottom:=5.0}
                                                                                                Widget:=stack_box:
                                                                                                    Orientation:=orientation.Horizontal
                                                                                                    Slots:=array:

                                                                                                        #Base Lv text
                                                                                                        stack_box_slot:
                                                                                                            VerticalAlignment:=vertical_alignment.Center
                                                                                                            HorizontalAlignment:=horizontal_alignment.Center
                                                                                                            #Distribution:=option{0.33}
                                                                                                            Padding:=margin{Left:=0.0, Top:=5.0, Right:=8.0, Bottom:=0.0}
                                                                                                            Widget:=text_block{DefaultText := StringToMessage("BASE LV"), DefaultTextColor:=MakeColorFromHex("#FFFFFFFF"), DefaultShadowColor:=MakeColorFromHex("#000000FF"), DefaultShadowOffset:=option{vector2{X:=3.0, Y:=3.0}}}

                                                                                            #Second horizontal stack box (Actually an overlay)
                                                                                            stack_box_slot:
                                                                                                VerticalAlignment:=vertical_alignment.Center
                                                                                                HorizontalAlignment:=horizontal_alignment.Center
                                                                                                Distribution:=option{1.0}
                                                                                                Padding:=margin{Left:=0.0, Top:=0.0, Right:=0.0, Bottom:=0.0}
                                                                                                Widget:=overlay:
                                                                                                    Slots:=array:

                                                                                                        #Base description text 1
                                                                                                        overlay_slot:
                                                                                                            VerticalAlignment:=vertical_alignment.Center
                                                                                                            HorizontalAlignment:=horizontal_alignment.Center
                                                                                                            #Distribution:=option{0.33}
                                                                                                            Padding:=margin{Left:=0.0, Top:=0.0, Right:=0.0, Bottom:=0.0}
                                                                                                            Widget:=stack_box:
                                                                                                                Orientation:=orientation.Horizontal
                                                                                                                Slots:=array:

                                                                                                                    stack_box_slot:
                                                                                                                        VerticalAlignment:=vertical_alignment.Center
                                                                                                                        HorizontalAlignment:=horizontal_alignment.Center
                                                                                                                        #Distribution:=option{0.33}
                                                                                                                        Padding:=margin{Left:=0.0, Top:=0.0, Right:=7.0, Bottom:=5.0}
                                                                                                                        Widget:=texture_block{DefaultImage:=Textures.Boosters.T_IncreaseMaxShield, DefaultDesiredSize:=vector2{X:=25.0, Y:=25.0}, DefaultTint:=MakeColorFromHex("#FFFFFFFF")}

                                                                                                                    stack_box_slot:
                                                                                                                        VerticalAlignment:=vertical_alignment.Center
                                                                                                                        HorizontalAlignment:=horizontal_alignment.Center
                                                                                                                        #Distribution:=option{0.33}
                                                                                                                        Padding:=margin{Left:=0.0, Top:=0.0, Right:=20.0, Bottom:=5.0}
                                                                                                                        Widget:=text_block{DefaultText := StringToMessage("+50%"), DefaultTextColor:=MakeColorFromHex("#c3c3c3FF"), DefaultShadowColor:=MakeColorFromHex("#000000FF"), DefaultShadowOffset:=option{vector2{X:=3.0, Y:=3.0}}}
                                                                                                        
                                                                                                        #Base description text 2
                                                                                                        overlay_slot:
                                                                                                            VerticalAlignment:=vertical_alignment.Bottom
                                                                                                            HorizontalAlignment:=horizontal_alignment.Center
                                                                                                            #Distribution:=option{0.33}
                                                                                                            Padding:=margin{Left:=0.0, Top:=20.0, Right:=0.0, Bottom:=0.0}
                                                                                                            Widget:=stack_box:
                                                                                                                Orientation:=orientation.Horizontal
                                                                                                                Slots:=array:

                                                                                                                    stack_box_slot:
                                                                                                                        VerticalAlignment:=vertical_alignment.Center
                                                                                                                        HorizontalAlignment:=horizontal_alignment.Fill
                                                                                                                        Distribution:=option{1.0}
                                                                                                                        Padding:=margin{Left:=0.0, Top:=0.0, Right:=30.0, Bottom:=0.0}
                                                                                                                        Widget:=text_block{DefaultText := StringToMessage("once"), DefaultTextColor:=MakeColorFromHex("#c3c3c3FF"), DefaultShadowColor:=MakeColorFromHex("#000000FF"), DefaultShadowOffset:=option{vector2{X:=3.0, Y:=3.0}}}

                                                                                #Arrow
                                                                                stack_box_slot:
                                                                                    VerticalAlignment:=vertical_alignment.Center
                                                                                    HorizontalAlignment:=horizontal_alignment.Center
                                                                                    #Distribution:=option{0.33}
                                                                                    Padding:=margin{Left:=8.0, Top:=0.0, Right:=12.0, Bottom:=3.0}
                                                                                    Widget:=overlay:
                                                                                        Slots:=array:
                                                                                            
                                                                                            #Arrow
                                                                                            overlay_slot:
                                                                                                VerticalAlignment:=vertical_alignment.Center
                                                                                                HorizontalAlignment:=horizontal_alignment.Center
                                                                                                #Distribution:=option{0.33}
                                                                                                Padding:=margin{Left:=0.0, Top:=0.0, Right:=0.0, Bottom:=3.0}
                                                                                                Widget:=texture_block{DefaultImage:=Textures.General.T_GenArrowHead, DefaultDesiredSize:=vector2{X:=26.0, Y:=24.0}, DefaultTint:=MakeColorFromHex("#FFFFFFFF")}
                                                                                
                                                                                #Second description column
                                                                                stack_box_slot:
                                                                                    VerticalAlignment:=vertical_alignment.Center
                                                                                    HorizontalAlignment:=horizontal_alignment.Right
                                                                                    Distribution:=option{1.0}
                                                                                    Padding:=margin{Left:=0.0, Top:=0.0, Right:=0.0, Bottom:=0.0}
                                                                                    Widget:=stack_box:
                                                                                        Orientation:=orientation.Vertical
                                                                                        Slots:=array:

                                                                                            #First horizontal stack box
                                                                                            stack_box_slot:
                                                                                                VerticalAlignment:=vertical_alignment.Center
                                                                                                HorizontalAlignment:=horizontal_alignment.Center
                                                                                                Distribution:=option{1.0}
                                                                                                Padding:=margin{Left:=0.0, Top:=0.0, Right:=0.0, Bottom:=5.0}
                                                                                                Widget:=stack_box:
                                                                                                    Orientation:=orientation.Horizontal
                                                                                                    Slots:=array:

                                                                                                        #Lv 1 text
                                                                                                        stack_box_slot:
                                                                                                            VerticalAlignment:=vertical_alignment.Center
                                                                                                            HorizontalAlignment:=horizontal_alignment.Center
                                                                                                            #Distribution:=option{0.33}
                                                                                                            Padding:=margin{Left:=0.0, Top:=5.0, Right:=8.0, Bottom:=0.0}
                                                                                                            Widget:=text_block{DefaultText := StringToMessage("LV. 1"), DefaultTextColor:=MakeColorFromHex("#f6fe17FF"), DefaultShadowColor:=MakeColorFromHex("#000000FF"), DefaultShadowOffset:=option{vector2{X:=3.0, Y:=3.0}}}
                                                                                                        
                                                                                            #Second horizontal stack box (Actually an overlay)
                                                                                            stack_box_slot:
                                                                                                VerticalAlignment:=vertical_alignment.Center
                                                                                                HorizontalAlignment:=horizontal_alignment.Center
                                                                                                Distribution:=option{1.0}
                                                                                                Padding:=margin{Left:=0.0, Top:=0.0, Right:=0.0, Bottom:=0.0}
                                                                                                Widget:=overlay:
                                                                                                    Slots:=array:

                                                                                                        #Base description text 1
                                                                                                        overlay_slot:
                                                                                                            VerticalAlignment:=vertical_alignment.Center
                                                                                                            HorizontalAlignment:=horizontal_alignment.Center
                                                                                                            #Distribution:=option{0.33}
                                                                                                            Padding:=margin{Left:=0.0, Top:=0.0, Right:=0.0, Bottom:=0.0}
                                                                                                            Widget:=stack_box:
                                                                                                                Orientation:=orientation.Horizontal
                                                                                                                Slots:=array:

                                                                                                                    stack_box_slot:
                                                                                                                        VerticalAlignment:=vertical_alignment.Center
                                                                                                                        HorizontalAlignment:=horizontal_alignment.Center
                                                                                                                        #Distribution:=option{0.33}
                                                                                                                        Padding:=margin{Left:=0.0, Top:=0.0, Right:=7.0, Bottom:=5.0}
                                                                                                                        Widget:=texture_block{DefaultImage:=Textures.Boosters.T_IncreaseMaxShield, DefaultDesiredSize:=vector2{X:=25.0, Y:=25.0}, DefaultTint:=MakeColorFromHex("#FFFFFFFF")}

                                                                                                                    stack_box_slot:
                                                                                                                        VerticalAlignment:=vertical_alignment.Center
                                                                                                                        HorizontalAlignment:=horizontal_alignment.Center
                                                                                                                        #Distribution:=option{0.33}
                                                                                                                        Padding:=margin{Left:=0.0, Top:=0.0, Right:=20.0, Bottom:=5.0}
                                                                                                                        Widget:=text_block{DefaultText := StringToMessage("+50%"), DefaultTextColor:=MakeColorFromHex("#f6fe17FF"), DefaultShadowColor:=MakeColorFromHex("#000000FF"), DefaultShadowOffset:=option{vector2{X:=3.0, Y:=3.0}}}
                                                                                                                    
                                                                                                        #Base description text 2
                                                                                                        overlay_slot:
                                                                                                            VerticalAlignment:=vertical_alignment.Bottom
                                                                                                            HorizontalAlignment:=horizontal_alignment.Center
                                                                                                            #Distribution:=option{0.33}
                                                                                                            Padding:=margin{Left:=0.0, Top:=20.0, Right:=0.0, Bottom:=0.0}
                                                                                                            Widget:=stack_box:
                                                                                                                Orientation:=orientation.Horizontal
                                                                                                                Slots:=array:

                                                                                                                    stack_box_slot:
                                                                                                                        VerticalAlignment:=vertical_alignment.Center
                                                                                                                        HorizontalAlignment:=horizontal_alignment.Left
                                                                                                                        Distribution:=option{1.0}
                                                                                                                        Padding:=margin{Left:=0.0, Top:=0.0, Right:=30.0, Bottom:=0.0}
                                                                                                                        Widget:=text_block{DefaultText := StringToMessage("twice"), DefaultTextColor:=MakeColorFromHex("#f6fe17FF"), DefaultShadowColor:=MakeColorFromHex("#000000FF"), DefaultShadowOffset:=option{vector2{X:=3.0, Y:=3.0}}}
                                                                                                                    
                                                        #Spacing
                                                        stack_box_slot:
                                                            VerticalAlignment:=vertical_alignment.Center
                                                            HorizontalAlignment:=horizontal_alignment.Right
                                                            #Distribution:=option{1.0}
                                                            Padding:=margin{Left:=0.0, Top:=0.0, Right:=0.0, Bottom:=0.0}
                                                            Widget:=texture_block{DefaultImage:=Textures.General.T_GenSquTrans, DefaultDesiredSize:=vector2{X:=350.0, Y:=120.0}, DefaultTint:=MakeColorFromHex("#00FF00FF")}
                                                        
                                                        #Call to action
                                                        stack_box_slot:
                                                            VerticalAlignment:=vertical_alignment.Center
                                                            HorizontalAlignment:=horizontal_alignment.Right
                                                            #Distribution:=option{1.0}
                                                            Padding:=margin{Left:=0.0, Top:=0.0, Right:=0.0, Bottom:=0.0}
                                                            Widget:=overlay:
                                                                Slots:=array:

                                                                    #Overlay structure, first is the button itself
                                                                    overlay_slot:
                                                                        VerticalAlignment:=vertical_alignment.Center
                                                                        HorizontalAlignment:=horizontal_alignment.Center
                                                                        #Distribution:=option{1.0}
                                                                        Padding:=margin{Left:=0.0, Top:=0.0, Right:=0.0, Bottom:=0.0}
                                                                        Widget:=button_loud{DefaultText:=StringToMessage("UPGRADE\n ")}

                                                                    #Overlay structure, icon and value
                                                                    overlay_slot:
                                                                        VerticalAlignment:=vertical_alignment.Fill
                                                                        HorizontalAlignment:=horizontal_alignment.Center
                                                                        #Distribution:=option{1.0}
                                                                        Padding:=margin{Left:=0.0, Top:=0.0, Right:=0.0, Bottom:=0.0}
                                                                        Widget:=stack_box:
                                                                            Orientation:=orientation.Horizontal
                                                                            Slots:=array:

                                                                                #Icon and value
                                                                                stack_box_slot:
                                                                                    VerticalAlignment:=vertical_alignment.Bottom
                                                                                    HorizontalAlignment:=horizontal_alignment.Right
                                                                                    #Distribution:=option{1.0}
                                                                                    Padding:=margin{Left:=0.0, Top:=0.0, Right:=0.0, Bottom:=8.0}
                                                                                    Widget:=texture_block{DefaultImage:=Textures.Currency.T_Gold, DefaultDesiredSize:=vector2{X:=35.0, Y:=35.0}, DefaultTint:=MakeColorFromHex("#000000FF")}                

                                                                                stack_box_slot:
                                                                                    VerticalAlignment:=vertical_alignment.Bottom
                                                                                    HorizontalAlignment:=horizontal_alignment.Center
                                                                                    #Distribution:=option{1.0}
                                                                                    Padding:=margin{Left:=5.0, Top:=0.0, Right:=0.0, Bottom:=5.0}
                                                                                    Widget:=text_block{DefaultText := StringToMessage("15"), DefaultTextColor:=MakeColorFromHex("#000000FF"), DefaultShadowColor:=MakeColorFromHex("#000000FF"), DefaultShadowOffset:=option{vector2{X:=3.0, Y:=3.0}}}

            NavPurchase5:=overlay:
                Slots:=array:

                    #Makeshift size box value
                    overlay_slot:
                        VerticalAlignment:=vertical_alignment.Center
                        HorizontalAlignment:=horizontal_alignment.Center
                        #Distribution:=option{0.33}
                        Padding:=margin{Left:=0.0, Top:=0.0, Right:=0.0, Bottom:=0.0}
                        Widget:=texture_block{DefaultImage:=Textures.General.T_GenSquTrans, DefaultDesiredSize:=vector2{X:=1800.0, Y:=760.0}, DefaultTint:=MakeColorFromHex("#00GG00FF")}

                    #Main content structure
                    overlay_slot:
                        VerticalAlignment:=vertical_alignment.Center
                        HorizontalAlignment:=horizontal_alignment.Center
                        #Distribution:=option{0.33}
                        Padding:=margin{Left:=0.0, Top:=0.0, Right:=0.0, Bottom:=0.0}
                        Widget:=stack_box:
                            Orientation:=orientation.Vertical
                            Slots:=array:

                                #Single row layout
                                stack_box_slot:
                                    VerticalAlignment:=vertical_alignment.Center
                                    HorizontalAlignment:=horizontal_alignment.Center
                                    #Distribution:=option{0.33}
                                    Padding:=margin{Left:=0.0, Top:=0.0, Right:=0.0, Bottom:=0.0}
                                    Widget:=overlay:
                                        Slots:=array:
                                            
                                            #Makeshift size box
                                                overlay_slot:
                                                    VerticalAlignment:=vertical_alignment.Center
                                                    HorizontalAlignment:=horizontal_alignment.Center
                                                    #Distribution:=option{0.33}
                                                    Padding:=margin{Left:=0.0, Top:=0.0, Right:=0.0, Bottom:=0.0}
                                                    Widget:=texture_block{DefaultImage:=Textures.General.T_GenSquTrans, DefaultDesiredSize:=vector2{X:=1600.0, Y:=120.0}, DefaultTint:=MakeColorFromHex("#FF0000FF")}        

                                            #Row layout
                                            overlay_slot:
                                                VerticalAlignment:=vertical_alignment.Center
                                                HorizontalAlignment:=horizontal_alignment.Fill
                                                #Distribution:=option{0.33}
                                                Padding:=margin{Left:=0.0, Top:=0.0, Right:=0.0, Bottom:=0.0}
                                                Widget:=stack_box:
                                                    Orientation:=orientation.Horizontal
                                                    Slots:=array:
                                                        
                                                        #Booster icon and rarity
                                                        stack_box_slot:
                                                            VerticalAlignment:=vertical_alignment.Center
                                                            HorizontalAlignment:=horizontal_alignment.Left
                                                            Distribution:=option{1.0}
                                                            Padding:=margin{Left:=0.0, Top:=0.0, Right:=0.0, Bottom:=0.0}
                                                            Widget:=overlay:
                                                                Slots:=array:
                                                                    
                                                                    #Booster Rarity BG
                                                                    overlay_slot:
                                                                        VerticalAlignment:=vertical_alignment.Center
                                                                        HorizontalAlignment:=horizontal_alignment.Left
                                                                        #Distribution:=option{1.0}
                                                                        Padding:=margin{Left:=0.0, Top:=0.0, Right:=0.0, Bottom:=0.0}
                                                                        Widget:=texture_block{DefaultImage:=Textures.Rarity.T_IconBG_Common, DefaultDesiredSize:=vector2{X:=120.0, Y:=120.0}, DefaultTint:=MakeColorFromHex("#FFFFFFFF")}
                                                                    
                                                                    #Booster Icon
                                                                    overlay_slot:
                                                                        VerticalAlignment:=vertical_alignment.Center
                                                                        HorizontalAlignment:=horizontal_alignment.Left
                                                                        #Distribution:=option{1.0}
                                                                        Padding:=margin{Left:=0.0, Top:=0.0, Right:=0.0, Bottom:=0.0}
                                                                        Widget:=texture_block{DefaultImage:=Textures.Boosters.T_IncreaseMaxShield, DefaultDesiredSize:=vector2{X:=116.0, Y:=116.0}, DefaultTint:=MakeColorFromHex("#FFFFFFFF")}

                                                                    #Booster Border
                                                                    overlay_slot:
                                                                        VerticalAlignment:=vertical_alignment.Center
                                                                        HorizontalAlignment:=horizontal_alignment.Left
                                                                        #Distribution:=option{1.0}
                                                                        Padding:=margin{Left:=0.0, Top:=0.0, Right:=0.0, Bottom:=0.0}
                                                                        Widget:=texture_block{DefaultImage:=Textures.General.T_GenericBorderSquare, DefaultDesiredSize:=vector2{X:=120.0, Y:=120.0}, DefaultTint:=MakeColorFromHex("#cdc953FF")}
                                                                                                                                                                                                        
                                                        #Booster description
                                                        stack_box_slot:
                                                            VerticalAlignment:=vertical_alignment.Center
                                                            HorizontalAlignment:=horizontal_alignment.Fill
                                                            #Distribution:=option{1.0}
                                                            Padding:=margin{Left:=0.0, Top:=0.0, Right:=0.0, Bottom:=0.0}
                                                            Widget:=overlay:
                                                                Slots:=array:

                                                                    #BG color
                                                                    overlay_slot:
                                                                        VerticalAlignment:=vertical_alignment.Center
                                                                        HorizontalAlignment:=horizontal_alignment.Center
                                                                        #Distribution:=option{0.33}
                                                                        Padding:=margin{Left:=0.0, Top:=0.0, Right:=0.0, Bottom:=0.0}
                                                                        Widget:=texture_block{DefaultImage:=Textures.General.T_GenericSquare, DefaultDesiredSize:=vector2{X:=570.0, Y:=120.0}, DefaultTint:=MakeColorFromHex("#031126FF")}

                                                                    #Overall Border
                                                                    overlay_slot:
                                                                        VerticalAlignment:=vertical_alignment.Center
                                                                        HorizontalAlignment:=horizontal_alignment.Center
                                                                        #Distribution:=option{0.33}
                                                                        Padding:=margin{Left:=0.0, Top:=0.0, Right:=0.0, Bottom:=0.0}
                                                                        Widget:=texture_block{DefaultImage:=Textures.General.T_LevelBorder, DefaultDesiredSize:=vector2{X:=570.0, Y:=120.0}, DefaultTint:=MakeColorFromHex("#1d4b89FF")}
                                                                    
                                                                        
                                                                        
                                                                    #Description Header
                                                                    overlay_slot:
                                                                        VerticalAlignment:=vertical_alignment.Top
                                                                        HorizontalAlignment:=horizontal_alignment.Center
                                                                        #Distribution:=option{0.33}
                                                                        Padding:=margin{Left:=0.0, Top:=0.0, Right:=0.0, Bottom:=5.0}
                                                                        Widget:=text_block{DefaultText := StringToMessage("Increase Max Shield"), DefaultTextColor:=MakeColorFromHex("#f6fe17FF"), DefaultShadowColor:=MakeColorFromHex("#000000FF"), DefaultShadowOffset:=option{vector2{X:=3.0, Y:=3.0}}}

                                                                    #Description Header
                                                                    overlay_slot:
                                                                        VerticalAlignment:=vertical_alignment.Bottom
                                                                        HorizontalAlignment:=horizontal_alignment.Center
                                                                        #Distribution:=option{0.33}
                                                                        Padding:=margin{Left:=0.0, Top:=0.0, Right:=0.0, Bottom:=0.0}
                                                                        Widget:=text_block{DefaultText := StringToMessage("If health falls under 40%,\nrecover 50% shield once"), DefaultTextColor:=MakeColorFromHex("#c4c5c7FF"), DefaultShadowColor:=MakeColorFromHex("#000000FF"), DefaultShadowOffset:=option{vector2{X:=3.0, Y:=3.0}}}

                                                        #Upgrade description
                                                        stack_box_slot:
                                                            VerticalAlignment:=vertical_alignment.Center
                                                            HorizontalAlignment:=horizontal_alignment.Center
                                                            #Distribution:=option{0.33}
                                                            Padding:=margin{Left:=0.0, Top:=0.0, Right:=0.0, Bottom:=0.0}
                                                            Widget:=overlay:
                                                                Slots:=array:

                                                                    #BG color
                                                                    overlay_slot:
                                                                        VerticalAlignment:=vertical_alignment.Center
                                                                        HorizontalAlignment:=horizontal_alignment.Center
                                                                        #Distribution:=option{0.33}
                                                                        Padding:=margin{Left:=0.0, Top:=0.0, Right:=0.0, Bottom:=0.0}
                                                                        Widget:=texture_block{DefaultImage:=Textures.General.T_GenericSquare, DefaultDesiredSize:=vector2{X:=348.0, Y:=120.0}, DefaultTint:=MakeColorFromHex("#000000FF")}

                                                                    #Overall Border
                                                                    overlay_slot:
                                                                        VerticalAlignment:=vertical_alignment.Center
                                                                        HorizontalAlignment:=horizontal_alignment.Center
                                                                        #Distribution:=option{0.33}
                                                                        Padding:=margin{Left:=0.0, Top:=0.0, Right:=0.0, Bottom:=0.0}
                                                                        Widget:=texture_block{DefaultImage:=Textures.General.T_LevelBorder, DefaultDesiredSize:=vector2{X:=348.0, Y:=120.0}, DefaultTint:=MakeColorFromHex("#1d4b89FF")}

                                                                    #Upgrade description layout - Horizontal stack box
                                                                    overlay_slot:
                                                                        VerticalAlignment:=vertical_alignment.Center
                                                                        HorizontalAlignment:=horizontal_alignment.Fill
                                                                        #Distribution:=option{1.0}
                                                                        Padding:=margin{Left:=0.0, Top:=0.0, Right:=0.0, Bottom:=5.0}
                                                                        Widget:=stack_box:
                                                                            Orientation:=orientation.Horizontal
                                                                            Slots:=array:

                                                                                #First description column
                                                                                stack_box_slot:
                                                                                    VerticalAlignment:=vertical_alignment.Center
                                                                                    HorizontalAlignment:=horizontal_alignment.Left
                                                                                    Distribution:=option{1.0}
                                                                                    Padding:=margin{Left:=25.0, Top:=0.0, Right:=0.0, Bottom:=0.0}
                                                                                    Widget:=stack_box:
                                                                                        Orientation:=orientation.Vertical
                                                                                        Slots:=array:

                                                                                            #First horizontal stack box
                                                                                            stack_box_slot:
                                                                                                VerticalAlignment:=vertical_alignment.Center
                                                                                                HorizontalAlignment:=horizontal_alignment.Center
                                                                                                Distribution:=option{1.0}
                                                                                                Padding:=margin{Left:=0.0, Top:=0.0, Right:=0.0, Bottom:=5.0}
                                                                                                Widget:=stack_box:
                                                                                                    Orientation:=orientation.Horizontal
                                                                                                    Slots:=array:

                                                                                                        #Base Lv text
                                                                                                        stack_box_slot:
                                                                                                            VerticalAlignment:=vertical_alignment.Center
                                                                                                            HorizontalAlignment:=horizontal_alignment.Center
                                                                                                            #Distribution:=option{0.33}
                                                                                                            Padding:=margin{Left:=0.0, Top:=5.0, Right:=8.0, Bottom:=0.0}
                                                                                                            Widget:=text_block{DefaultText := StringToMessage("BASE LV"), DefaultTextColor:=MakeColorFromHex("#FFFFFFFF"), DefaultShadowColor:=MakeColorFromHex("#000000FF"), DefaultShadowOffset:=option{vector2{X:=3.0, Y:=3.0}}}

                                                                                            #Second horizontal stack box (Actually an overlay)
                                                                                            stack_box_slot:
                                                                                                VerticalAlignment:=vertical_alignment.Center
                                                                                                HorizontalAlignment:=horizontal_alignment.Center
                                                                                                Distribution:=option{1.0}
                                                                                                Padding:=margin{Left:=0.0, Top:=0.0, Right:=0.0, Bottom:=0.0}
                                                                                                Widget:=overlay:
                                                                                                    Slots:=array:

                                                                                                        #Base description text 1
                                                                                                        overlay_slot:
                                                                                                            VerticalAlignment:=vertical_alignment.Center
                                                                                                            HorizontalAlignment:=horizontal_alignment.Center
                                                                                                            #Distribution:=option{0.33}
                                                                                                            Padding:=margin{Left:=0.0, Top:=0.0, Right:=0.0, Bottom:=0.0}
                                                                                                            Widget:=stack_box:
                                                                                                                Orientation:=orientation.Horizontal
                                                                                                                Slots:=array:

                                                                                                                    stack_box_slot:
                                                                                                                        VerticalAlignment:=vertical_alignment.Center
                                                                                                                        HorizontalAlignment:=horizontal_alignment.Center
                                                                                                                        #Distribution:=option{0.33}
                                                                                                                        Padding:=margin{Left:=0.0, Top:=0.0, Right:=7.0, Bottom:=5.0}
                                                                                                                        Widget:=texture_block{DefaultImage:=Textures.Boosters.T_IncreaseMaxShield, DefaultDesiredSize:=vector2{X:=25.0, Y:=25.0}, DefaultTint:=MakeColorFromHex("#FFFFFFFF")}

                                                                                                                    stack_box_slot:
                                                                                                                        VerticalAlignment:=vertical_alignment.Center
                                                                                                                        HorizontalAlignment:=horizontal_alignment.Center
                                                                                                                        #Distribution:=option{0.33}
                                                                                                                        Padding:=margin{Left:=0.0, Top:=0.0, Right:=20.0, Bottom:=5.0}
                                                                                                                        Widget:=text_block{DefaultText := StringToMessage("+50%"), DefaultTextColor:=MakeColorFromHex("#c3c3c3FF"), DefaultShadowColor:=MakeColorFromHex("#000000FF"), DefaultShadowOffset:=option{vector2{X:=3.0, Y:=3.0}}}
                                                                                                        
                                                                                                        #Base description text 2
                                                                                                        overlay_slot:
                                                                                                            VerticalAlignment:=vertical_alignment.Bottom
                                                                                                            HorizontalAlignment:=horizontal_alignment.Center
                                                                                                            #Distribution:=option{0.33}
                                                                                                            Padding:=margin{Left:=0.0, Top:=20.0, Right:=0.0, Bottom:=0.0}
                                                                                                            Widget:=stack_box:
                                                                                                                Orientation:=orientation.Horizontal
                                                                                                                Slots:=array:

                                                                                                                    stack_box_slot:
                                                                                                                        VerticalAlignment:=vertical_alignment.Center
                                                                                                                        HorizontalAlignment:=horizontal_alignment.Fill
                                                                                                                        Distribution:=option{1.0}
                                                                                                                        Padding:=margin{Left:=0.0, Top:=0.0, Right:=30.0, Bottom:=0.0}
                                                                                                                        Widget:=text_block{DefaultText := StringToMessage("once"), DefaultTextColor:=MakeColorFromHex("#c3c3c3FF"), DefaultShadowColor:=MakeColorFromHex("#000000FF"), DefaultShadowOffset:=option{vector2{X:=3.0, Y:=3.0}}}

                                                                                #Arrow
                                                                                stack_box_slot:
                                                                                    VerticalAlignment:=vertical_alignment.Center
                                                                                    HorizontalAlignment:=horizontal_alignment.Center
                                                                                    #Distribution:=option{0.33}
                                                                                    Padding:=margin{Left:=8.0, Top:=0.0, Right:=12.0, Bottom:=3.0}
                                                                                    Widget:=overlay:
                                                                                        Slots:=array:
                                                                                            
                                                                                            #Arrow
                                                                                            overlay_slot:
                                                                                                VerticalAlignment:=vertical_alignment.Center
                                                                                                HorizontalAlignment:=horizontal_alignment.Center
                                                                                                #Distribution:=option{0.33}
                                                                                                Padding:=margin{Left:=0.0, Top:=0.0, Right:=0.0, Bottom:=3.0}
                                                                                                Widget:=texture_block{DefaultImage:=Textures.General.T_GenArrowHead, DefaultDesiredSize:=vector2{X:=26.0, Y:=24.0}, DefaultTint:=MakeColorFromHex("#FFFFFFFF")}
                                                                                
                                                                                #Second description column
                                                                                stack_box_slot:
                                                                                    VerticalAlignment:=vertical_alignment.Center
                                                                                    HorizontalAlignment:=horizontal_alignment.Right
                                                                                    Distribution:=option{1.0}
                                                                                    Padding:=margin{Left:=0.0, Top:=0.0, Right:=0.0, Bottom:=0.0}
                                                                                    Widget:=stack_box:
                                                                                        Orientation:=orientation.Vertical
                                                                                        Slots:=array:

                                                                                            #First horizontal stack box
                                                                                            stack_box_slot:
                                                                                                VerticalAlignment:=vertical_alignment.Center
                                                                                                HorizontalAlignment:=horizontal_alignment.Center
                                                                                                Distribution:=option{1.0}
                                                                                                Padding:=margin{Left:=0.0, Top:=0.0, Right:=0.0, Bottom:=5.0}
                                                                                                Widget:=stack_box:
                                                                                                    Orientation:=orientation.Horizontal
                                                                                                    Slots:=array:

                                                                                                        #Lv 1 text
                                                                                                        stack_box_slot:
                                                                                                            VerticalAlignment:=vertical_alignment.Center
                                                                                                            HorizontalAlignment:=horizontal_alignment.Center
                                                                                                            #Distribution:=option{0.33}
                                                                                                            Padding:=margin{Left:=0.0, Top:=5.0, Right:=8.0, Bottom:=0.0}
                                                                                                            Widget:=text_block{DefaultText := StringToMessage("LV. 1"), DefaultTextColor:=MakeColorFromHex("#f6fe17FF"), DefaultShadowColor:=MakeColorFromHex("#000000FF"), DefaultShadowOffset:=option{vector2{X:=3.0, Y:=3.0}}}
                                                                                                        
                                                                                            #Second horizontal stack box (Actually an overlay)
                                                                                            stack_box_slot:
                                                                                                VerticalAlignment:=vertical_alignment.Center
                                                                                                HorizontalAlignment:=horizontal_alignment.Center
                                                                                                Distribution:=option{1.0}
                                                                                                Padding:=margin{Left:=0.0, Top:=0.0, Right:=0.0, Bottom:=0.0}
                                                                                                Widget:=overlay:
                                                                                                    Slots:=array:

                                                                                                        #Base description text 1
                                                                                                        overlay_slot:
                                                                                                            VerticalAlignment:=vertical_alignment.Center
                                                                                                            HorizontalAlignment:=horizontal_alignment.Center
                                                                                                            #Distribution:=option{0.33}
                                                                                                            Padding:=margin{Left:=0.0, Top:=0.0, Right:=0.0, Bottom:=0.0}
                                                                                                            Widget:=stack_box:
                                                                                                                Orientation:=orientation.Horizontal
                                                                                                                Slots:=array:

                                                                                                                    stack_box_slot:
                                                                                                                        VerticalAlignment:=vertical_alignment.Center
                                                                                                                        HorizontalAlignment:=horizontal_alignment.Center
                                                                                                                        #Distribution:=option{0.33}
                                                                                                                        Padding:=margin{Left:=0.0, Top:=0.0, Right:=7.0, Bottom:=5.0}
                                                                                                                        Widget:=texture_block{DefaultImage:=Textures.Boosters.T_IncreaseMaxShield, DefaultDesiredSize:=vector2{X:=25.0, Y:=25.0}, DefaultTint:=MakeColorFromHex("#FFFFFFFF")}

                                                                                                                    stack_box_slot:
                                                                                                                        VerticalAlignment:=vertical_alignment.Center
                                                                                                                        HorizontalAlignment:=horizontal_alignment.Center
                                                                                                                        #Distribution:=option{0.33}
                                                                                                                        Padding:=margin{Left:=0.0, Top:=0.0, Right:=20.0, Bottom:=5.0}
                                                                                                                        Widget:=text_block{DefaultText := StringToMessage("+50%"), DefaultTextColor:=MakeColorFromHex("#f6fe17FF"), DefaultShadowColor:=MakeColorFromHex("#000000FF"), DefaultShadowOffset:=option{vector2{X:=3.0, Y:=3.0}}}
                                                                                                                    
                                                                                                        #Base description text 2
                                                                                                        overlay_slot:
                                                                                                            VerticalAlignment:=vertical_alignment.Bottom
                                                                                                            HorizontalAlignment:=horizontal_alignment.Center
                                                                                                            #Distribution:=option{0.33}
                                                                                                            Padding:=margin{Left:=0.0, Top:=20.0, Right:=0.0, Bottom:=0.0}
                                                                                                            Widget:=stack_box:
                                                                                                                Orientation:=orientation.Horizontal
                                                                                                                Slots:=array:

                                                                                                                    stack_box_slot:
                                                                                                                        VerticalAlignment:=vertical_alignment.Center
                                                                                                                        HorizontalAlignment:=horizontal_alignment.Left
                                                                                                                        Distribution:=option{1.0}
                                                                                                                        Padding:=margin{Left:=0.0, Top:=0.0, Right:=30.0, Bottom:=0.0}
                                                                                                                        Widget:=text_block{DefaultText := StringToMessage("twice"), DefaultTextColor:=MakeColorFromHex("#f6fe17FF"), DefaultShadowColor:=MakeColorFromHex("#000000FF"), DefaultShadowOffset:=option{vector2{X:=3.0, Y:=3.0}}}
                                                                                                                    
                                                        #Spacing
                                                        stack_box_slot:
                                                            VerticalAlignment:=vertical_alignment.Center
                                                            HorizontalAlignment:=horizontal_alignment.Right
                                                            #Distribution:=option{1.0}
                                                            Padding:=margin{Left:=0.0, Top:=0.0, Right:=0.0, Bottom:=0.0}
                                                            Widget:=texture_block{DefaultImage:=Textures.General.T_GenSquTrans, DefaultDesiredSize:=vector2{X:=350.0, Y:=120.0}, DefaultTint:=MakeColorFromHex("#00FF00FF")}
                                                        
                                                        #Call to action
                                                        stack_box_slot:
                                                            VerticalAlignment:=vertical_alignment.Center
                                                            HorizontalAlignment:=horizontal_alignment.Right
                                                            #Distribution:=option{1.0}
                                                            Padding:=margin{Left:=0.0, Top:=0.0, Right:=0.0, Bottom:=0.0}
                                                            Widget:=overlay:
                                                                Slots:=array:

                                                                    #Overlay structure, first is the button itself
                                                                    overlay_slot:
                                                                        VerticalAlignment:=vertical_alignment.Center
                                                                        HorizontalAlignment:=horizontal_alignment.Center
                                                                        #Distribution:=option{1.0}
                                                                        Padding:=margin{Left:=0.0, Top:=0.0, Right:=0.0, Bottom:=0.0}
                                                                        Widget:=button_loud{DefaultText:=StringToMessage("UPGRADE\n ")}

                                                                    #Overlay structure, icon and value
                                                                    overlay_slot:
                                                                        VerticalAlignment:=vertical_alignment.Fill
                                                                        HorizontalAlignment:=horizontal_alignment.Center
                                                                        #Distribution:=option{1.0}
                                                                        Padding:=margin{Left:=0.0, Top:=0.0, Right:=0.0, Bottom:=0.0}
                                                                        Widget:=stack_box:
                                                                            Orientation:=orientation.Horizontal
                                                                            Slots:=array:

                                                                                #Icon and value
                                                                                stack_box_slot:
                                                                                    VerticalAlignment:=vertical_alignment.Bottom
                                                                                    HorizontalAlignment:=horizontal_alignment.Right
                                                                                    #Distribution:=option{1.0}
                                                                                    Padding:=margin{Left:=0.0, Top:=0.0, Right:=0.0, Bottom:=8.0}
                                                                                    Widget:=texture_block{DefaultImage:=Textures.Currency.T_Gold, DefaultDesiredSize:=vector2{X:=35.0, Y:=35.0}, DefaultTint:=MakeColorFromHex("#000000FF")}                

                                                                                stack_box_slot:
                                                                                    VerticalAlignment:=vertical_alignment.Bottom
                                                                                    HorizontalAlignment:=horizontal_alignment.Center
                                                                                    #Distribution:=option{1.0}
                                                                                    Padding:=margin{Left:=5.0, Top:=0.0, Right:=0.0, Bottom:=5.0}
                                                                                    Widget:=text_block{DefaultText := StringToMessage("15"), DefaultTextColor:=MakeColorFromHex("#000000FF"), DefaultShadowColor:=MakeColorFromHex("#000000FF"), DefaultShadowOffset:=option{vector2{X:=3.0, Y:=3.0}}}

            #Build of the screen
            HeaderCanvas:canvas = canvas:
                Slots:=array:
                    #Defines the parameters for the canvas' slot
                    canvas_slot:
                        Anchors := anchors{Minimum := vector2{X := 0.0, Y := 0.0}, Maximum := vector2{X := 1.0, Y := 1.0}}
                        Alignment := vector2{X := 0.5, Y := 0.5}
                        Widget:=overlay:
                            Slots:=array:

                                #Base BG
                                overlay_slot:
                                    VerticalAlignment:=vertical_alignment.Center
                                    HorizontalAlignment:=horizontal_alignment.Center
                                    Padding:=margin{Left:=0.0, Top:=0.0, Right:=0.0, Bottom:=0.0}
                                    #Store header
                                    Widget:=texture_block{DefaultImage:=Textures.General.T_Background, DefaultDesiredSize:=vector2{X:=1920.0, Y:=1080.0}}
                                
                                #Actual layout
                                overlay_slot:
                                    VerticalAlignment:=vertical_alignment.Fill
                                    HorizontalAlignment:=horizontal_alignment.Center
                                    Padding:=margin{Left:=0.0, Top:=0.0, Right:=0.0, Bottom:=30.0}
                                    #Store header
                                    Widget:=stack_box:
                                        Orientation:=orientation.Vertical
                                        Slots:=array:

                                            stack_box_slot:
                                                VerticalAlignment:=vertical_alignment.Top
                                                HorizontalAlignment:=horizontal_alignment.Center
                                                Padding:=margin{Left:=0.0, Top:=0.0, Right:=0.0, Bottom:=0.0}
                                                Distribution:=option{1.0}
                                                #Store header
                                                Widget:=BaseHeader
                            
                                            stack_box_slot:
                                                VerticalAlignment:=vertical_alignment.Center
                                                HorizontalAlignment:=horizontal_alignment.Center
                                                Padding:=margin{Left:=0.0, Top:=0.0, Right:=0.0, Bottom:=0.0}
                                                Distribution:=option{1.0}
                                                #Subtitle and pager
                                                Widget:=BoosterPager

                                            stack_box_slot:
                                                VerticalAlignment:=vertical_alignment.Center
                                                HorizontalAlignment:=horizontal_alignment.Center
                                                Padding:=margin{Left:=0.0, Top:=3.0, Right:=0.0, Bottom:=0.0}
                                                Distribution:=option{1.0}
                                                #Screen content
                                                Widget:=NavPurchase1

                                            stack_box_slot:
                                                VerticalAlignment:=vertical_alignment.Center
                                                HorizontalAlignment:=horizontal_alignment.Center
                                                Padding:=margin{Left:=0.0, Top:=3.0, Right:=0.0, Bottom:=0.0}
                                                Distribution:=option{1.0}
                                                #Screen content
                                                Widget:=NavPurchase2

                                            stack_box_slot:
                                                VerticalAlignment:=vertical_alignment.Center
                                                HorizontalAlignment:=horizontal_alignment.Center
                                                Padding:=margin{Left:=0.0, Top:=3.0, Right:=0.0, Bottom:=0.0}
                                                Distribution:=option{1.0}
                                                #Screen content
                                                Widget:=NavPurchase3

                                            stack_box_slot:
                                                VerticalAlignment:=vertical_alignment.Center
                                                HorizontalAlignment:=horizontal_alignment.Center
                                                Padding:=margin{Left:=0.0, Top:=3.0, Right:=0.0, Bottom:=0.0}
                                                Distribution:=option{1.0}
                                                #Screen content
                                                Widget:=NavPurchase4

                                            stack_box_slot:
                                                VerticalAlignment:=vertical_alignment.Center
                                                HorizontalAlignment:=horizontal_alignment.Center
                                                Padding:=margin{Left:=0.0, Top:=3.0, Right:=0.0, Bottom:=0.0}
                                                Distribution:=option{1.0}
                                                #Screen content
                                                Widget:=NavPurchase5

            if(PlayerUI:=GetPlayerUI[Player]):
                PlayerUI.AddWidget(HeaderCanvas)