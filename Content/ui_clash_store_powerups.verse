using { /Fortnite.com/Devices }
using { /Fortnite.com/Characters }
using { /Fortnite.com/UI}
using { /Verse.org/Simulation }
using { /Verse.org/Assets }
using { /Verse.org/Colors }
using { /Verse.org/Colors/NamedColors }
using { /UnrealEngine.com/Temporary/UI }
using { /UnrealEngine.com/Temporary/SpatialMath }

ui_powerup_card_data:=struct:
    StackBoxSlot:stack_box_slot
    Value:text_block
    Level:text_block

#Cbray: TODO: This can be turn to Struct
ui_powerup_store_widet_data:=class():
    Button_Close_Widget:button_loud=button_loud{}
    var Buttons:[]button_loud=array{}
    var ButtonsQuiet:[]button_quiet=array{}
    var UpgradeLevel:[]text_block=array{} #Number of upgrades
    var UpgradeValue:[]text_block=array{} #Value (Same as displayed on HUD)
    var CurrencyText:text_block=text_block{}

ui_on_click_card_data:=struct:
    Index:int
    Data:ui_powerup_store_widet_data

ui_clash_store_powerups := class(creative_device):

    @editable
    UpgradeText:string="UPGRADE"

    @editable
    MaxedOutText:string="MAXED OUT"

    @editable
    OpenStoreTrigger:trigger_device=trigger_device{}

    @editable
    ClashPowerupsManager:clash_powerup_manager_device=clash_powerup_manager_device{}

    @editable
    StoreCameraBlue:gameplay_camera_fixed_point_device=gameplay_camera_fixed_point_device{}

    @editable
    StoreCameraRed:gameplay_camera_fixed_point_device=gameplay_camera_fixed_point_device{}

    @editable
    CurrencyManager:clash_currency_manager=clash_currency_manager{}

    #Events
    CloseWidgetEvent:event(player)=event(player){}

    OnBegin<override>()<suspends>:void=
        OpenStoreTrigger.TriggeredEvent.Subscribe(OpenWidget)
        
    CreateCanvas(Player:player):void=

        PowerupWidgetData:=ui_powerup_store_widet_data{}
        PowerupWidgetData.Button_Close_Widget.SetText(StringToMessage("X"))
        PowerupWidgetData.Button_Close_Widget.OnClick().Subscribe(OnCloseButtonClicked)

        ButtonsStackBox:=stack_box:
            Orientation:=orientation.Horizontal

        PowerupCardsStackBox:=stack_box:
            Orientation:=orientation.Horizontal

        var ButtonIndex:int=0

        PlayerCurrency:=CurrencyManager.GetCurrency(Player)

        for(PowerupID:powerups_array):
           #Generate Powerup card

            TempCard:=GeneratePowerupCard(PowerupID)
            PowerupCardsStackBox.AddWidget(TempCard.StackBoxSlot)

            MaybePowerupDevice:=ClashPowerupsManager.GetPowerUpByID(PowerupID)

            if:
                ValidPowerupDevice:=MaybePowerupDevice?
            then:
                #Powerup Upgrade Level

                MaxUpgrade:= ClashPowerupsManager.GetPowerupMaxUpgrade(PowerupID)
                var CurrentLevel:int=0

                if:
                    CurrentUpgradeLevel:=ClashPowerupsManager.ClashPlayerDataDevice.GetPlayerPowerupUpgradeLevel[Player, PowerupID]
                then:
                    set CurrentLevel = CurrentUpgradeLevel
                    Print("Level: {CurrentLevel}")

                TempCard.Level.SetText(StringToMessage("Lvl\n{CurrentLevel}"))
                set PowerupWidgetData.UpgradeLevel+=array{TempCard.Level}

                #Powerup Upgrade Value

                CurrentValue:=ValidPowerupDevice.GetUIPowerupValueByLevel(CurrentLevel)
                TempCard.Value.SetText(StringToMessage("{CurrentValue}"))
                set PowerupWidgetData.UpgradeValue+=array{TempCard.Value}

                Print("Value: {CurrentValue}")

                #Button Info
                Cost:=ValidPowerupDevice.UpgradeCost
                StringCost:=ToTruncatedString(Cost, 0)
                TempButton:=button_loud{DefaultText:=StringToMessage("{UpgradeText}\n${StringCost}")}
                TempButtonQuiet:=button_quiet{DefaultText:=StringToMessage(MaxedOutText)}


                #Cbray: TODO: Mix all this checks in one single logic

                if:         
                    CurrentLevel >= MaxUpgrade
                then:
                    ToggleButtons(TempButtonQuiet, TempButton)
                    TempButtonQuiet.SetText(StringToMessage(MaxedOutText))
                else:
                    if(PlayerCurrency < Cost):
                        ToggleButtons(TempButtonQuiet, TempButton)
                        TempButtonQuiet.SetText(StringToMessage("{UpgradeText}\n${StringCost}"))
                    else:
                        ToggleButtons(TempButton, TempButtonQuiet)

                set PowerupWidgetData.Buttons+=array{TempButton}
                set PowerupWidgetData.ButtonsQuiet+=array{TempButtonQuiet}       

                ButtonsOverlay:=overlay{
                    Slots:=array{
                        overlay_slot:
                            Widget:=TempButton
                            VerticalAlignment:=vertical_alignment.Center
                            HorizontalAlignment:=horizontal_alignment.Center

                        overlay_slot:
                            Widget:=TempButtonQuiet
                            VerticalAlignment:=vertical_alignment.Center
                            HorizontalAlignment:=horizontal_alignment.Center                        
                    }
                }

                Slot:=stack_box_slot{
                    Widget:=ButtonsOverlay
                    VerticalAlignment:=vertical_alignment.Center
                    HorizontalAlignment:=horizontal_alignment.Center
                    Distribution:=option{0.2}
                    Padding:=margin{Left:=0.0, Top:=0.0, Right:=0.0, Bottom:=0.0}
                }

                CardData:=ui_on_click_card_data{
                    Index:=ButtonIndex
                    Data:=PowerupWidgetData
                }

                ButtonsStackBox.AddWidget(Slot)
                TempButton.OnClick().SubscribeButton(OnButtonClicked, CardData)
                set ButtonIndex+=1

        #Currency
        StringCurrency:=ToTruncatedString(PlayerCurrency, 0)
        set PowerupWidgetData.CurrencyText=text_block{DefaultText := StringToMessage("  {StringCurrency}"), DefaultTextColor:=MakeColorFromHex("#000000FF"), DefaultShadowColor:=MakeColorFromHex("#000000FF"), DefaultShadowOffset:=option{vector2{X:=3.0, Y:=3.0}}}

        #Main structure for store header
        BaseHeader:=overlay:
            Slots:=array:

                #Makeshift size box
                overlay_slot:
                    VerticalAlignment:=vertical_alignment.Center
                    HorizontalAlignment:=horizontal_alignment.Center
                    Padding:=margin{Left:=0.0, Top:=0.0, Right:=0.0, Bottom:=0.0}
                    Widget:=texture_block{DefaultImage:=Textures.General.T_GenSquTrans, DefaultDesiredSize:=vector2{X:=1800.0, Y:=100.0}, DefaultTint:=MakeColorFromHex("#FF0000FF")}

                #Header Layout
                overlay_slot:
                    VerticalAlignment:=vertical_alignment.Center
                    HorizontalAlignment:=horizontal_alignment.Fill
                    Padding:=margin{Left:=0.0, Top:=0.0, Right:=0.0, Bottom:=0.0}
                    Widget:=stack_box:
                        Orientation:=orientation.Horizontal
                        Slots:=array:
                            
                            #Coin$ counter
                            stack_box_slot:
                                VerticalAlignment:=vertical_alignment.Center
                                HorizontalAlignment:=horizontal_alignment.Left
                                Distribution:=option{0.33}
                                Padding:=margin{Left:=30.0, Top:=0.0, Right:=0.0, Bottom:=0.0}
                                Widget:=overlay:
                                    Slots:=array:

                                        #Makeshift size box
                                        #overlay_slot:
                                            #VerticalAlignment:=vertical_alignment.Center
                                            #HorizontalAlignment:=horizontal_alignment.Center
                                            ##Distribution:=option{0.33}
                                            #Padding:=margin{Left:=0.0, Top:=0.0, Right:=0.0, Bottom:=0.0}
                                            #Widget:=texture_block{DefaultImage:=Textures.General.T_GenericSquare, DefaultDesiredSize:=vector2{X:=650.0, Y:=50.0}, DefaultTint:=MakeColorFromHex("#FFFFFFFF")}

                                        #Coin$ counter
                                        overlay_slot:
                                            VerticalAlignment:=vertical_alignment.Center
                                            HorizontalAlignment:=horizontal_alignment.Left
                                            #Distribution:=option{0.33}
                                            Padding:=margin{Left:=0.0, Top:=0.0, Right:=0.0, Bottom:=0.0}
                                            Widget:=texture_block{DefaultImage:=Textures.Currency.T_GoldSlot, DefaultDesiredSize:=vector2{X:=180.0, Y:=50.0}, DefaultTint:=MakeColorFromHex("#FFFFFFFF")}

                                        #Coin$ counter layout
                                        overlay_slot:
                                            VerticalAlignment:=vertical_alignment.Center
                                            HorizontalAlignment:=horizontal_alignment.Left
                                            #Distribution:=option{1.0}
                                            Padding:=margin{Left:=0.0, Top:=0.0, Right:=0.0, Bottom:=0.0}
                                            Widget:=stack_box:
                                                Orientation:=orientation.Horizontal
                                                Slots:=array:

                                                    #Coin$ icon
                                                    stack_box_slot:
                                                        VerticalAlignment:=vertical_alignment.Center
                                                        HorizontalAlignment:=horizontal_alignment.Left
                                                        Distribution:=option{1.0}
                                                        Padding:=margin{Left:=2.0, Top:=0.0, Right:=0.0, Bottom:=0.0}
                                                        Widget:=texture_block{DefaultImage:=Textures.Currency.T_Gold, DefaultDesiredSize:=vector2{X:=40.0, Y:=40.0}, DefaultTint:=MakeColorFromHex("#FFFFFFFF")}

                                                    #Coin$ value
                                                    stack_box_slot:
                                                        VerticalAlignment:=vertical_alignment.Center
                                                        HorizontalAlignment:=horizontal_alignment.Left
                                                        Distribution:=option{1.0}
                                                        Padding:=margin{Left:=5.0, Top:=0.0, Right:=0.0, Bottom:=0.0}
                                                        Widget:=PowerupWidgetData.CurrencyText

                            #Header
                            stack_box_slot:
                                VerticalAlignment:=vertical_alignment.Center
                                HorizontalAlignment:=horizontal_alignment.Center
                                Distribution:=option{0.33}
                                Padding:=margin{Left:=0.0, Top:=0.0, Right:=0.0, Bottom:=0.0}
                                Widget:=text_block{DefaultText := StringToMessage("Permanent\nPOWERUPS"), DefaultTextColor:=MakeColorFromHex("#FFFFFFFF"), DefaultShadowColor:=MakeColorFromHex("#000000FF"), DefaultShadowOffset:=option{vector2{X:=3.0, Y:=3.0}}}

                            #Close Button
                            stack_box_slot:
                                VerticalAlignment:=vertical_alignment.Center
                                HorizontalAlignment:=horizontal_alignment.Right
                                Distribution:=option{0.33}
                                Padding:=margin{Left:=0.0, Top:=0.0, Right:=30.0, Bottom:=0.0}
                                Widget:=PowerupWidgetData.Button_Close_Widget


        #Maint content layout
        PowerupsContent:=overlay:
            Slots:=array:

                #Makeshift Size Box
                overlay_slot:
                    VerticalAlignment:=vertical_alignment.Center
                    HorizontalAlignment:=horizontal_alignment.Center
                    #Distribution:=option{0.33}
                    Padding:=margin{Left:=0.0, Top:=0.0, Right:=0.0, Bottom:=0.0}
                    Widget:=texture_block{DefaultImage:=Textures.General.T_GenSquTrans, DefaultDesiredSize:=vector2{X:=1800.0, Y:=710.0}, DefaultTint:=MakeColorFromHex("#00FF00FF")}

                #Powerups layout/Cards
                overlay_slot:
                    VerticalAlignment:=vertical_alignment.Center
                    HorizontalAlignment:=horizontal_alignment.Fill
                    #Distribution:=option{0.33}
                    Padding:=margin{Left:=0.0, Top:=0.0, Right:=0.0, Bottom:=0.0}
                    Widget:=PowerupCardsStackBox

        #Main content prompts
        NavPurchase:=overlay:
            Slots:=array:

                #Makeshift size box value
                overlay_slot:
                    VerticalAlignment:=vertical_alignment.Center
                    HorizontalAlignment:=horizontal_alignment.Center
                    #Distribution:=option{0.33}
                    Padding:=margin{Left:=0.0, Top:=0.0, Right:=0.0, Bottom:=0.0}
                    Widget:=texture_block{DefaultImage:=Textures.General.T_GenSquTrans, DefaultDesiredSize:=vector2{X:=1800.0, Y:=210.0}, DefaultTint:=MakeColorFromHex("#00GG00FF")}

                #Horizontal stack box layout
                overlay_slot:
                    VerticalAlignment:=vertical_alignment.Center
                    HorizontalAlignment:=horizontal_alignment.Fill
                    #Distribution:=option{1.0}
                    Padding:=margin{Left:=0.0, Top:=0.0, Right:=0.0, Bottom:=0.0}
                    Widget:=ButtonsStackBox

        #Build of the screen
        HeaderCanvas:canvas = canvas:
            Slots:=array:
                #Defines the parameters for the canvas' slot
                canvas_slot:
                    Anchors := anchors{Minimum := vector2{X := 0.0, Y := 0.0}, Maximum := vector2{X := 1.0, Y := 1.0}}
                    Alignment := vector2{X := 0.5, Y := 0.5}
                    Widget:=overlay:
                        Slots:=array:

                            #Base BG
                            overlay_slot:
                                VerticalAlignment:=vertical_alignment.Center
                                HorizontalAlignment:=horizontal_alignment.Center
                                Padding:=margin{Left:=0.0, Top:=0.0, Right:=0.0, Bottom:=0.0}
                                #Store header
                                Widget:=texture_block{DefaultImage:=Textures.General.T_Background, DefaultDesiredSize:=vector2{X:=1920.0, Y:=1080.0}}
                            
                            #Actual layout
                            overlay_slot:
                                VerticalAlignment:=vertical_alignment.Center
                                HorizontalAlignment:=horizontal_alignment.Center
                                Padding:=margin{Left:=0.0, Top:=0.0, Right:=0.0, Bottom:=0.0}
                                #Store header
                                Widget:=stack_box:
                                    Orientation:=orientation.Vertical
                                    Slots:=array:

                                        stack_box_slot:
                                            VerticalAlignment:=vertical_alignment.Top
                                            HorizontalAlignment:=horizontal_alignment.Center
                                            Padding:=margin{Left:=0.0, Top:=0.0, Right:=0.0, Bottom:=0.0}
                                            #Store header
                                            Widget:=BaseHeader
                        
                                        stack_box_slot:
                                            VerticalAlignment:=vertical_alignment.Center
                                            HorizontalAlignment:=horizontal_alignment.Center
                                            Padding:=margin{Left:=0.0, Top:=0.0, Right:=0.0, Bottom:=0.0}
                                            #Subtitle and pager
                                            Widget:=PowerupsContent

                                        stack_box_slot:
                                            VerticalAlignment:=vertical_alignment.Center
                                            HorizontalAlignment:=horizontal_alignment.Center
                                            Padding:=margin{Left:=0.0, Top:=0.0, Right:=0.0, Bottom:=0.0}
                                            #Screen content
                                            Widget:=NavPurchase
        #Setup Cortuines and add to Player UI      
        if(PlayerUI:=GetPlayerUI[Player]):
            PlayerUI.AddWidget(HeaderCanvas, player_ui_slot{InputMode:=ui_input_mode.All})

        spawn{CloseWidget(Player, HeaderCanvas)}

    OnButtonClicked(Message:widget_message, CardData:ui_on_click_card_data):void=
        Player:=Message.Player
        Index:=CardData.Index
        PowerupID:=IndexToPowerupID(Index)
        var UpgradeIndex:int=0
        MaxUpgrade:= ClashPowerupsManager.GetPowerupMaxUpgrade(PowerupID)
        Print("MaxUpgrade: {MaxUpgrade}")
        #If Fails there's not data for this powerup
        if:
            CurrentUpgradeLevel:=ClashPowerupsManager.ClashPlayerDataDevice.GetPlayerPowerupUpgradeLevel[Player, PowerupID]
        then:
            set UpgradeIndex = CurrentUpgradeLevel

        Print("Current level: {UpgradeIndex}")

        if:         
            MaxUpgrade > UpgradeIndex
        else:
            Print("Max Level reached!")
            return
       
        MaybePowerupDevice:=ClashPowerupsManager.GetPowerUpByID(PowerupID)

        if:        
            ValidPowerupDevice:=MaybePowerupDevice?
            TargetLevel:=CardData.Data.UpgradeLevel[Index]
            TargetValue:=CardData.Data.UpgradeValue[Index]
            TargetLoudButton:=CardData.Data.Buttons[Index]
            TargetQuietButton:=CardData.Data.ButtonsQuiet[Index]
        then:
            UpgradeCost:=ValidPowerupDevice.UpgradeCost
            PlayerCurrency:=CurrencyManager.GetCurrency(Player)

            if(PlayerCurrency >= UpgradeCost):
                NewPlayerCurrency:=PlayerCurrency - UpgradeCost
                StringCurrency:=ToNotationString(NewPlayerCurrency)
                CurrencyManager.AddCurrency(-UpgradeCost, Player, currency_id.None)
                CardData.Data.CurrencyText.SetText(StringToMessage("  {StringCurrency}"))

                set UpgradeIndex+=1
                Print("New level: {UpgradeIndex}")
                ClashPowerupsManager.ClashPlayerDataDevice.SetPlayerPowerupUpgradeLevel(Message.Player, PowerupID, UpgradeIndex)
                ClashPowerupsManager.GivePowerUpToPlayer(Message.Player, PowerupID, UpgradeIndex)

                #Update UI           
                TargetLevel.SetText(StringToMessage("Lvl\n{UpgradeIndex}"))
                CurrentValue:=ValidPowerupDevice.GetUIPowerupValueByLevel(UpgradeIndex)
                TargetValue.SetText(StringToMessage("{CurrentValue}"))

                #TODO: Move Upgrade level check to Refersh method

                CurrentVisibility:=TargetLoudButton.GetVisibility()

                if:         
                    UpgradeIndex >= MaxUpgrade
                    CurrentVisibility = widget_visibility.Visible
                then:
                    ToggleButtons(TargetQuietButton, TargetLoudButton)

                RefreshAllButtonsInfo(Player, CardData)
                 
    GeneratePowerupCard(PowerupID:powerup_id):ui_powerup_card_data=

        Value:=text_block{DefaultText:=StringToMessage("VXXX"), DefaultTextColor:=MakeColorFromHex("#000000FF")}
        Level:=text_block{DefaultText := StringToMessage("Lvl\nXX"), DefaultTextColor:=MakeColorFromHex("#000000FF")}
        var DisplayName:message=StringToMessage("Powerup Name")
        var Texture:texture=PowerUpSystem.Textures.Icons.T_BlankPowerUp

        MaybePowerup:=ClashPowerupsManager.GetPowerUpByID(PowerupID)
            if:
                ValidPowerup:=MaybePowerup?
            then:
                set Texture = ValidPowerup.PowerupIcon
                set DisplayName = StringToMessage(ValidPowerup.PowerupDisplayName)

        #Card Layout - Max health
        CardSlot:=stack_box_slot:
            VerticalAlignment:=vertical_alignment.Center
            HorizontalAlignment:=horizontal_alignment.Fill
            Distribution:=option{0.21}
            Padding:=margin{Left:=0.0, Top:=0.0, Right:=0.0, Bottom:=0.0}
            Widget:=overlay:
                Slots:=array:

                    #Makeshift size box, including offset for level bubble and price frame
                    overlay_slot:
                        VerticalAlignment:=vertical_alignment.Center
                        HorizontalAlignment:=horizontal_alignment.Center
                        #Distribution:=option{0.33}
                        Padding:=margin{Left:=0.0, Top:=0.0, Right:=0.0, Bottom:=5.0}
                        Widget:=texture_block{DefaultImage:=Textures.General.T_GenSquTrans, DefaultDesiredSize:=vector2{X:=300.0, Y:=320.0}, DefaultTint:=MakeColorFromHex("#FF0000FF")}

                    #Base background for each card
                    overlay_slot:
                        VerticalAlignment:=vertical_alignment.Center
                        HorizontalAlignment:=horizontal_alignment.Center
                        #Distribution:=option{0.33}
                        Padding:=margin{Left:=0.0, Top:=15.0, Right:=0.0, Bottom:=5.0}
                        Widget:=texture_block{DefaultImage:=Textures.General.T_GenericSquare, DefaultDesiredSize:=vector2{X:=285.0, Y:=370.0}, DefaultTint:=MakeColorFromHex("#000e21FF")}

                    #Border for each card
                    overlay_slot:
                        VerticalAlignment:=vertical_alignment.Center
                        HorizontalAlignment:=horizontal_alignment.Center
                        #Distribution:=option{0.33}
                        Padding:=margin{Left:=0.0, Top:=15.0, Right:=0.0, Bottom:=5.0}
                        Widget:=texture_block{DefaultImage:=Textures.General.T_GenericBorderSquare, DefaultDesiredSize:=vector2{X:=285.0, Y:=370.0}, DefaultTint:=MakeColorFromHex("#00c2ffFF")}

                    #Layout for the card itself - 1
                    overlay_slot:
                        VerticalAlignment:=vertical_alignment.Top
                        HorizontalAlignment:=horizontal_alignment.Center
                        #Distribution:=option{0.33}
                        Padding:=margin{Left:=0.0, Top:=80.0, Right:=0.0, Bottom:=5.0}
                        Widget:=stack_box:
                            Orientation:=orientation.Vertical
                            Slots:=array:

                                #Header area
                                stack_box_slot:
                                    VerticalAlignment:=vertical_alignment.Center
                                    HorizontalAlignment:=horizontal_alignment.Center
                                    #Distribution:=option{0.33}
                                    Padding:=margin{Left:=0.0, Top:=0.0, Right:=0.0, Bottom:=45.0}
                                    Widget:=overlay:
                                        Slots:=array:

                                            #Header frame
                                            overlay_slot:
                                                VerticalAlignment:=vertical_alignment.Center
                                                HorizontalAlignment:=horizontal_alignment.Center
                                                #Distribution:=option{0.33}
                                                Padding:=margin{Left:=0.0, Top:=0.0, Right:=0.0, Bottom:=0.0}
                                                Widget:=texture_block{DefaultImage:=Textures.General.T_GenericSquare, DefaultDesiredSize:=vector2{X:=280.0, Y:=60.0}, DefaultTint:=MakeColorFromHex("#00c2ffFF")}

                                            #Actual header text
                                            overlay_slot:
                                                VerticalAlignment:=vertical_alignment.Center
                                                HorizontalAlignment:=horizontal_alignment.Center
                                                #Distribution:=option{0.33}
                                                Padding:=margin{Left:=0.0, Top:=0.0, Right:=0.0, Bottom:=0.0}
                                                Widget:=text_block{DefaultText := DisplayName, DefaultTextColor:=MakeColorFromHex("#000000FF")}

                                #Card icon
                                stack_box_slot:
                                    VerticalAlignment:=vertical_alignment.Center
                                    HorizontalAlignment:=horizontal_alignment.Center
                                    #Distribution:=option{0.33}
                                    Padding:=margin{Left:=0.0, Top:=0.0, Right:=0.0, Bottom:=25.0}
                                    Widget:=texture_block{DefaultImage:=Texture, DefaultDesiredSize:=vector2{X:=190.0, Y:=190.0}, DefaultTint:=MakeColorFromHex("#FFFFFFFF")}

                                #Card boost value
                                stack_box_slot:
                                    VerticalAlignment:=vertical_alignment.Center
                                    HorizontalAlignment:=horizontal_alignment.Center
                                    #Distribution:=option{0.33}
                                    Padding:=margin{Left:=0.0, Top:=0.0, Right:=0.0, Bottom:=0.0}
                                    Widget:=overlay:
                                        Slots:=array:

                                            #Boost value frame
                                            overlay_slot:
                                                VerticalAlignment:=vertical_alignment.Center
                                                HorizontalAlignment:=horizontal_alignment.Center
                                                #Distribution:=option{0.33}
                                                Padding:=margin{Left:=0.0, Top:=0.0, Right:=0.0, Bottom:=0.0}
                                                Widget:=texture_block{DefaultImage:=Textures.General.T_GenericSquare, DefaultDesiredSize:=vector2{X:=150.0, Y:=50.0}, DefaultTint:=MakeColorFromHex("#FFFFFFFF")}

                                            #Boost value
                                            overlay_slot:
                                                VerticalAlignment:=vertical_alignment.Center
                                                HorizontalAlignment:=horizontal_alignment.Center
                                                #Distribution:=option{0.33}
                                                Padding:=margin{Left:=0.0, Top:=0.0, Right:=0.0, Bottom:=0.0}
                                                Widget:=Value

                    #Level circular frame at the top right corner
                    overlay_slot:
                        VerticalAlignment:=vertical_alignment.Top
                        HorizontalAlignment:=horizontal_alignment.Right
                        #Distribution:=option{0.33}
                        Padding:=margin{Left:=0.0, Top:=0.0, Right:=0.0, Bottom:=0.0}
                        Widget:=overlay:
                            Slots:=array:

                                #Circular frame texture
                                overlay_slot:
                                    VerticalAlignment:=vertical_alignment.Center
                                    HorizontalAlignment:=horizontal_alignment.Center
                                    #Distribution:=option{0.33}
                                    Padding:=margin{Left:=0.0, Top:=0.0, Right:=0.0, Bottom:=0.0}
                                    Widget:=texture_block{DefaultImage:=Textures.General.T_GenericCircle, DefaultDesiredSize:=vector2{X:=110.0, Y:=80.0}, DefaultTint:=MakeColorFromHex("#f7ff1aFF")}

                                #Circular frame text
                                overlay_slot:
                                    VerticalAlignment:=vertical_alignment.Center
                                    HorizontalAlignment:=horizontal_alignment.Center
                                    #Distribution:=option{0.33}
                                    Padding:=margin{Left:=0.0, Top:=0.0, Right:=0.0, Bottom:=0.0}
                                    Widget:=Level

        CardData:=ui_powerup_card_data{
                StackBoxSlot:=CardSlot
                Value:=Value
                Level:=Level
        }

        return CardData

    OpenWidget(Agent:?agent):void=
        if(Player:=player[Agent?]):
            CreateCanvas(Player)
            if(GetAgentTeam(Player) = team_id.Red):
                StoreCameraRed.AddTo(Player)
            else:
                StoreCameraBlue.AddTo(Player)
            if(FortCharacter:=fort_character[Player]):
                StasisArguments:=stasis_args{
                    AllowTurning:=false
                    AllowFalling:=false
                    AllowEmotes:=false
                }
                FortCharacter.PutInStasis(StasisArguments)
                FortCharacter.SetVulnerability(false)
                
    CloseWidget(Player:player, Canvas:canvas)<suspends>:void=

        if(Agent:=agent[Player], FortCharacter:=Agent.GetFortCharacter[]):
            Print("Cast OK")

            defer:
                Print("On Defer")
                if(PlayerUI:=GetPlayerUI[Player]):
                    PlayerUI.RemoveWidget(Canvas)
                    if(GetAgentTeam(Player) = team_id.Red):
                        StoreCameraRed.RemoveFrom(Player)
                    else:
                        StoreCameraBlue.RemoveFrom(Player)
                        
                    FortCharacter.ReleaseFromStasis()
                    FortCharacter.SetVulnerability(true)
            
            TerminateCorutine:event()=event(){}

            race:
                TerminateCorutine.Await()
                sync:
                    loop:             
                        FortCharacter.EliminatedEvent().Await()
                        TerminateCorutine.Signal()
                    loop:
                        TargetPlayer:=CloseWidgetEvent.Await()
                        if(TargetPlayer = Player):
                            TerminateCorutine.Signal()
   
    OnCloseButtonClicked(Message:widget_message):void=
        CloseWidgetEvent.Signal(Message.Player)

    ToggleButtons(OnButton:widget, OffButton:widget):void=
        OnButton.SetVisibility(widget_visibility.Visible)
        OffButton.SetVisibility(widget_visibility.Collapsed)

    #TODO: When Currency manager dispatch when a player got currency
    # HandleExternalEvents(Player:player, WidgetData:ui_powerup_store_widet_data)<suspends>:void=
    #     CloseHandlerEvent:=event(){}
    #     race:
    #         CloseHandlerEvent.Await()
    #         sync:
    #             loop:
    #                 TargetPlayer:=CloseWidgetEvent.Await()
    #                 if(TargetPlayer = Player):
    #                     CloseHandlerEvent.Signal()
    #             loop:
    #                 CurrencyManager. #OnCurrencyAdded

    RefreshButtonInfo(Player:player, PowerupID:powerup_id, CardData:ui_on_click_card_data):void=
        Index:=CardData.Index
        MaybePowerupDevice:=ClashPowerupsManager.GetPowerUpByID(PowerupID)
        if:
            ValidPowerupDevice:=MaybePowerupDevice?
            TargetLoudButton:=CardData.Data.Buttons[Index]
            TargetQuietButton:=CardData.Data.ButtonsQuiet[Index]
        then:
            UpgradeCost:=ValidPowerupDevice.UpgradeCost
            StringCost:=ToNotationString(UpgradeCost)
            PlayerCurrency:=CurrencyManager.GetCurrency(Player)

            if(PlayerCurrency < UpgradeCost):
                ToggleButtons(TargetQuietButton, TargetLoudButton)
                TargetQuietButton.SetText(StringToMessage("{UpgradeText}\n${StringCost}"))

    RefreshAllButtonsInfo(Player:player, CardData:ui_on_click_card_data):void=
        var Index:int=0
        for(PowerupID:powerups_array):
            TempCardData:=ui_on_click_card_data{Index:=Index, Data:=CardData.Data}
            RefreshButtonInfo(Player, PowerupID, TempCardData)
            set Index+=1

