
using { /Fortnite.com/Devices }
using { /Fortnite.com/UI }
using { /Verse.org/Simulation }
using { /Verse.org/Colors }
using { /UnrealEngine.com/Temporary/UI }
using { /UnrealEngine.com/Temporary/SpatialMath }

ui_common_data := struct():
    Canvas:canvas
    MainStackBox:stack_box

    # For the moment only one currenty, add stack box in the future if needed
    CurrencyTextBlock:text_block

    # Reference to show the message for not enough money
    NoMoneyTextBlock:text_block

<# 
UI that uses common template: header with title, currency, and back button
#>
ui_common := class(ui_base_device):
    #@editable
    #DisplayCurrency:currency_id = currency_id.Gold

    @editable
    CurrencyTextColor:color = NamedColors.Black

    @editable
    TitleTextColor:color = NamedColors.White

    @editable
    CloseString:string = "X"

    @editable
    BackButtonRightPadding:float = 150.0

    @editable
    SuccessSound : audio_player_device = audio_player_device{}

    @editable
    FailSound : audio_player_device = audio_player_device{}

    @editable
    NoMoneyString:string = "You don't have enough currency!"

    OnBackInputPressed(Agent:agent): void =
        if (Player:=player[Agent]):
            HideEvent.Signal(Player)

    ConstructBaseCanvas(Player:player, CurrencyManager:clash_currency_manager, TitleString:string, BackButtonString:string, 
    WidgetRemovedEvent:event(), ?bShowNotMoneyMessage:logic =false): ui_common_data =
        BackButton:button_loud = button_loud{DefaultText:=StringToMessage(BackButtonString)}
        spawn {AwaitBackButtonSelection(BackButton, WidgetRemovedEvent)}

        NoMoneyTextBlock:=text_block:
            DefaultText:=StringToMessage(NoMoneyString)
            DefaultTextColor:=NamedColors.Red
        
        NoMoneyTextBlock.SetVisibility(widget_visibility.Collapsed)

        CurrencyTextBlock:=text_block:
            #DefaultText:=StringToMessage("{ToNotationString(CurrencyManager.GetPlayerCurrency(Player, DisplayCurrency))}")
            DefaultTextColor:=CurrencyTextColor

        MainStackBox:=stack_box:
            Orientation:=orientation.Vertical
            Slots:=array:
                # Header
                stack_box_slot:
                    HorizontalAlignment:=horizontal_alignment.Fill
                    VerticalAlignment:=vertical_alignment.Fill
                    Widget:=overlay:
                        Slots:=array:
                            # Title
                            overlay_slot:
                                VerticalAlignment:=vertical_alignment.Top
                                HorizontalAlignment:=horizontal_alignment.Center
                                Padding:=margin{Top:=20.0, Left:=20.0, Right:=20.0, Bottom:=20.0}
                                Widget:=text_block{DefaultText:=StringToMessage(TitleString), DefaultTextColor:=TitleTextColor,
                                    DefaultJustification:=text_justification.Center}
                            # Back Button
                            overlay_slot:
                                VerticalAlignment:=vertical_alignment.Center
                                HorizontalAlignment:=horizontal_alignment.Right
                                Padding:=margin{Right:=BackButtonRightPadding}
                                Widget:=BackButton
                            # Currency Slot
                            overlay_slot:
                                VerticalAlignment:=vertical_alignment.Center
                                HorizontalAlignment:=horizontal_alignment.Left
                                Padding:=margin{Left:=150.0}
                                Widget:=overlay:
                                    Slots:=array:
                                        # Background
                                        <#overlay_slot:
                                            HorizontalAlignment:=horizontal_alignment.Fill
                                            VerticalAlignment:=vertical_alignment.Fill
                                            Widget:=texture_block{DefaultImage:=FindCurrencySlotBackground(DisplayCurrency), DefaultDesiredSize:=vector2{X:=300.0, Y:=65.0}}
                                        # Icon
                                        overlay_slot:
                                            HorizontalAlignment:=horizontal_alignment.Left
                                            VerticalAlignment:=vertical_alignment.Center
                                            Padding:=margin{Left:=15.0}
                                            Widget:=texture_block{DefaultImage:=FindCurrencyIcon(DisplayCurrency), DefaultDesiredSize:= vector2{X:=45.0, Y:=45.0}}#>
                                        # Text
                                        overlay_slot:
                                            HorizontalAlignment:=horizontal_alignment.Fill
                                            VerticalAlignment:=vertical_alignment.Center
                                            Padding:=margin{Left:=90.0, Right:=10.0}
                                            Widget:=CurrencyTextBlock

        Canvas:=canvas:
            Slots:=array:
                canvas_slot:
                    Anchors:=anchors{Minimum:=vector2{X:=0.0,Y:=0.0}, Maximum:=vector2{X:=1.0,Y:=1.0}}
                    Alignment:=vector2{X:=0.0, Y:=0.0}
                    Offsets:=margin{Top:=30.0, Left:=30.0, Right:=30.0, Bottom:=30.0}
                    Widget:=overlay:
                        Slots:=array:
                            # Full Background
                            overlay_slot:
                                VerticalAlignment:=vertical_alignment.Fill
                                HorizontalAlignment:=horizontal_alignment.Fill
                                Widget:=texture_block{DefaultImage:=Textures.General.T_Background}
                            # Main
                            overlay_slot:
                                VerticalAlignment:=vertical_alignment.Fill
                                HorizontalAlignment:=horizontal_alignment.Fill
                                Padding:=margin{Top:=30.0, Left:=30.0, Right:=30.0, Bottom:=30.0}
                                Widget:=MainStackBox
                            #No money message
                            overlay_slot:
                                VerticalAlignment:=vertical_alignment.Bottom
                                HorizontalAlignment:=horizontal_alignment.Center
                                Padding:=margin{Top:=50.0, Left:=30.0, Right:=30.0, Bottom:=0.0}
                                Widget:=NoMoneyTextBlock

        ui_common_data:
            Canvas:=Canvas
            MainStackBox:=MainStackBox
            CurrencyTextBlock:=CurrencyTextBlock
            NoMoneyTextBlock:=NoMoneyTextBlock

    <#AwaitCurrencyChanged(Player:player, CurrencyManager:clash_currency_manager, UIData:ui_common_data)<suspends>: void =
        CurrencyChangedData:=CurrencyManager.CurrencyChangedEvent.Await()
        if (CurrencyChangedData.Currency = DisplayCurrency and CurrencyChangedData.Player = Player):
            UIData.CurrencyTextBlock.SetText(StringToMessage("{ToNotationString(CurrencyChangedData.CurrentValue)}"))#>

    AwaitShowNoMoneyMessage(Player:player, UIData:ui_common_data)<suspends>: void =
        PlayerEvent:=NoCurrencyEvent.Await()
        if (PlayerEvent = Player):
            UIData.NoMoneyTextBlock.SetVisibility(widget_visibility.Visible)
            Sleep(2.0)
            UIData.NoMoneyTextBlock.SetVisibility(widget_visibility.Collapsed)

    AwaitBackButtonSelection(Button:text_button_base, WidgetRemovedEvent:event())<suspends>: void =
        race:
            block:
                WidgetRemovedEvent.Await()
                return
            block:
                WidgetMessage := Button.OnClick().Await()
                HideEvent.Signal(WidgetMessage.Player)

    PlayInteractionSound(bWasSusccesfull:logic, Agent:agent):void=
        SuccessSound.Stop(Agent)
        if (bWasSusccesfull?):
            SuccessSound.Play(Agent)
        else:
            FailSound.Play(Agent)