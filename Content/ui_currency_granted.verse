using { /Fortnite.com/Devices }
using { /Verse.org/Simulation }
using { /UnrealEngine.com/Temporary/Diagnostics }
using { /Fortnite.com/Characters }
using { /Fortnite.com/Game }
using { /UnrealEngine.com/Temporary/UI }
using { /UnrealEngine.com/Temporary/SpatialMath }
using { /Fortnite.com/UI }
using { /Verse.org/Colors }

currency_widget_data:=struct:
    Container:canvas
    Quantity:float
    CanvasPosition:int
    TimeElapsed:float = 0.0


#Shows the currency granted to player

ui_currency_granted := class(creative_device):
    @editable
    MessageLifeTime : float = 5.0

    @editable
    TimeStep : float = 0.1

    @editable
    CurrencyManager:clash_currency_manager = clash_currency_manager{}

    # in tuple the order is the container, time elapsed and BottomPadding
    var CurrencyEarnedPerPlayer : [player][]currency_widget_data = map{}

    var IsAlreadyInLoop : logic = false

    var CanvasPositions : []float = array{}

    var FreePositionsByPlayer : [player][]logic = map{} 
    
    var NewNotificationPosition : int = 0

    # --------------------------------------------------------------------------------
    OnBegin<override>()<suspends>:void =
        for (Index:=0..9):
            set CanvasPositions = CanvasPositions + array{45.0*Index}

        loop:
            Payload:=CurrencyManager.CurrencyChangedEvent.Await()
            AddCurrencyGranted(Payload.Player, Payload.CurrentValue-Payload.PreviousValue, Payload.CurrencySourceIcon)

    # --------------------------------------------------------------------------------
    AddCurrencyGranted(Player:player, Value:float, CurrencySource:currency_id): void =
        # No need to do anything if we are not earning
        if (Value <= 0.0):
            return

        # Logic is only relevant if the receiving player has valid UI
        if (PlayerUI := GetPlayerUI[Player]):

            # Make sure receiving player has an array of possible slot locations
            if (not FreePositionsByPlayer[Player]):
                CreateFreePostions(Player)

            # Create canvas for new gold message
            set NewNotificationPosition = ClaimLowestFreePosition(Player)
            NewNotificationCanvas := CreateCanvas(Value, NewNotificationPosition, CurrencySource)

            # Create the gold message data
            NewNotificationData := currency_widget_data:
                Container := NewNotificationCanvas
                Quantity := Value
                CanvasPosition := NewNotificationPosition

            if (PlayerNotifications := CurrencyEarnedPerPlayer[Player]):
                if (set CurrencyEarnedPerPlayer[Player] += array{NewNotificationData}) {}
            else:
                if (set CurrencyEarnedPerPlayer[Player] = array{NewNotificationData}) {}

            if (not IsAlreadyInLoop?):
                spawn{GoldMessageLoop()}
            
            PlayerUI.AddWidget(NewNotificationCanvas)

            if (Character:=Player.GetFortCharacter[]):
                Character.EliminatedEvent().Subscribe(OnCharacterEliminated)

    # --------------------------------------------------------------------------------
    FindLowestFreePosition(Player:player)<transacts>:int=
        for(Index->bIsFree : FreePositionsByPlayer[Player], bIsFree = true):
            return Index
        return 0

    # --------------------------------------------------------------------------------
    ClaimLowestFreePosition(Player:player):int=
        LowestPosition := FindLowestFreePosition(Player)
        SetPositionFreeState(Player, LowestPosition, false)
        return LowestPosition

    # --------------------------------------------------------------------------------
    SetPositionFreeState(Player:player, Position:int, bFreeState:logic):void=
        if (set FreePositionsByPlayer[Player][Position] = bFreeState) {}
        else:
            FreeStateString := if (bFreeState?) then "TRUE" else "FALSE"
            Print("Failed to set position free state!! Position: {Position}, Free State: {FreeStateString}")

    # --------------------------------------------------------------------------------
    CreateFreePostions(Player:player):void=
        var TempFree : []logic = array{}
        for(Indexes:CanvasPositions):
            set TempFree = TempFree + array{true}
        set FreePositionsByPlayer = ConcatenateMaps(FreePositionsByPlayer, map{Player => TempFree})

    # --------------------------------------------------------------------------------
    GoldMessageLoop()<suspends>:void=
        set IsAlreadyInLoop = true

        loop:
            Sleep(0.1)
            for(Player -> Notifications : CurrencyEarnedPerPlayer, PlayerUI:=GetPlayerUI[Player]):
                var NewNotificationsArray : []currency_widget_data = array{}
                for(Notification : Notifications):
                    NewTime := Notification.TimeElapsed + 0.1
                    if (NewTime >= MessageLifeTime):
                        if (set FreePositionsByPlayer[Player][Notification.CanvasPosition] = true) {}
                        PlayerUI.RemoveWidget(Notification.Container)
                    else:
                        NewNotification := currency_widget_data:
                            Container:=Notification.Container
                            Quantity:=Notification.Quantity
                            TimeElapsed:=NewTime
                            CanvasPosition:=Notification.CanvasPosition
                        set NewNotificationsArray = NewNotificationsArray + array{NewNotification}

                if (set CurrencyEarnedPerPlayer[Player] = NewNotificationsArray) {}

    # --------------------------------------------------------------------------------
    RemovePlayerFromMap(Player:player):[player][]currency_widget_data=
        var NewMap:[player][]currency_widget_data = map{}

        for (Key -> Value : CurrencyEarnedPerPlayer, Key <> Player):
          set NewMap = ConcatenateMaps(NewMap, map{Key => Value})

        return NewMap

    # --------------------------------------------------------------------------------
    OnCharacterEliminated(Result:elimination_result):void=
        if (KilledAgent := Result.EliminatedCharacter.GetAgent[], KilledPlayer := player[KilledAgent]):
            if (PlayerUI := GetPlayerUI[KilledPlayer]):
                for (UIDataArray:CurrencyEarnedPerPlayer[KilledPlayer]):
                    PlayerUI.RemoveWidget(UIDataArray.Container)

            var TempFree : []logic = array{}
            for (Indexes:CanvasPositions):
                set TempFree = TempFree + array{true}
            
            if (set FreePositionsByPlayer[KilledPlayer] = TempFree) {}   
            set CurrencyEarnedPerPlayer = RemovePlayerFromMap(KilledPlayer)

    # --------------------------------------------------------------------------------
    CreateCanvas(Quantity:float, Position:int, CurrencySource:currency_id):canvas=
        var CurrentLocation : float = 0.0
        if (NewLocation:=CanvasPositions[Position]):
            set CurrentLocation = NewLocation

        CurrencyColor:color = NamedColors.Yellow

        ContentOverlay:=overlay:
            Slots:=array:
                overlay_slot:
                    Widget:=overlay:
                        Slots:=array:
                            overlay_slot:
                                HorizontalAlignment := horizontal_alignment.Fill
                                VerticalAlignment := vertical_alignment.Fill
                                Widget:=color_block:
                                    DefaultColor:=CurrencyColor
                                    DefaultOpacity:=0.6
                            overlay_slot:
                                HorizontalAlignment := horizontal_alignment.Fill
                                VerticalAlignment := vertical_alignment.Fill
                                Padding := margin{Left:= 15.0, Top:=0.0, Right:=15.0, Bottom:=0.0}
                                Widget:=stack_box:
                                    Orientation:=orientation.Horizontal
                                    Slots:=array:
                                        stack_box_slot:
                                            VerticalAlignment := vertical_alignment.Center
                                            HorizontalAlignment := horizontal_alignment.Center
                                            Widget:=texture_block:
                                                DefaultImage:=FindKillSourceIcon(CurrencySource)
                                        stack_box_slot:
                                            VerticalAlignment := vertical_alignment.Center
                                            Widget:=text_block{DefaultText:=StringToMessage("+")}
                                        stack_box_slot:
                                            VerticalAlignment := vertical_alignment.Center
                                            HorizontalAlignment := horizontal_alignment.Center
                                            Widget:=texture_block:
                                                DefaultImage:=Textures.Currency.T_Gold
                                        stack_box_slot:
                                            VerticalAlignment := vertical_alignment.Center
                                            Widget:=text_block{DefaultText:=StringToMessage("{ToNotationString(Quantity)}")}
        WidgetCanvas:=canvas:
            Slots:=array:
                canvas_slot:
                    Anchors:=anchors{Minimum:=vector2{X:=0.59,Y:=0.59}, Maximum:=vector2{X:=0.59,Y:=0.59}}
                    Offsets := margin{Left:=0.0,Top:=CurrentLocation,Right:=0.0,Bottom:=0.0}
                    SizeToContent:=true
                    Widget:=ContentOverlay
        
       # spawn{UpdateWidgetPositionValue(WidgetCanvas, ContentOverlay)}
        return WidgetCanvas

    # --------------------------------------------------------------------------------
    # jcarrero - UpdateWidgetPositionValue function temporarily unused until we can find a better way to do it
    UpdateWidgetPositionValue(Canvas : canvas, ContentOverlay : overlay) <suspends>:void=  
        var YValue : float = NewNotificationPosition * 1.0
        
        loop:
            Sleep(0.0)
                set YValue -= 1.0

                ContentOverlay.SetVisibility(widget_visibility.Hidden)
                Canvas.RemoveWidget(ContentOverlay)
                ContentOverlay.SetVisibility(widget_visibility.Visible)

                Canvas.AddWidget(
                    Canvas_slot := canvas_slot:
                        Anchors:=anchors{Minimum:=vector2{X:=0.59,Y:=0.59}, Maximum:=vector2{X:=0.59,Y:=0.59}}
                        Offsets := margin{Left:=0.0,Top:=YValue, Right:=0.0,Bottom:=0.0}
                        SizeToContent:=true
                        Widget:=ContentOverlay        
                )
