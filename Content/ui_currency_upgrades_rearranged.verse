using { /Fortnite.com/UI}
using { /Fortnite.com/Devices }
using { /Verse.org/Simulation }
using { /Verse.org/Colors }
using { /Verse.org/Assets }
using { /UnrealEngine.com/Temporary/UI }
using { /UnrealEngine.com/Temporary/SpatialMath }
using { /UnrealEngine.com/Temporary/Diagnostics }

# See https://dev.epicgames.com/documentation/en-us/uefn/create-your-own-device-in-verse for how to create a verse device.

# A Verse-authored creative device that can be placed in a level
ui_currency_upgrades_rearranged := class(creative_device):#class(ui_common):
    
    

    # Runs when the device is started in a running game
    OnBegin<override>()<suspends>:void=
        # TODO: Replace this with your code
        Print("Hello, world!")
        Print("2 + 2 = {2 + 2}")
        CreateCanvas()
        
        
    CreateCanvas():void=
        Players:=GetPlayspace().GetPlayers()
        if(Player:=Players[0]):
                        
            #Defines several components to be used along the implementation:
            ImageBackground :=texture_block{DefaultImage:=Textures.General.T_Background}
            AssistTab :=texture_block{DefaultImage:=Textures.General.T_BlueTab, DefaultDesiredSize:=vector2{X:=359.0, Y:=56.0}}
            AssistTitle :=text_block{DefaultText:=StringToMessage("By Assist"), DefaultTextColor:=NamedColors.White}
            ElimTab :=texture_block{DefaultImage:=Textures.General.T_BlueTab, DefaultDesiredSize:=vector2{X:=359.0, Y:=56.0}}
            ElimTitle :=text_block{DefaultText:=StringToMessage("By Elimination"), DefaultTextColor:=NamedColors.White}
            BackButton:button_loud = button_loud{DefaultText:=StringToMessage("<")}
            EnemiesButton:button_loud = button_loud{DefaultText:=StringToMessage("ENEMIES")}
            BossesButton:button_loud = button_loud{DefaultText:=StringToMessage("BOSSES")}
            OthersButton:button_loud = button_loud{DefaultText:=StringToMessage("OTHERS")}
            PermanentPowerUpsCheck :=texture_block{DefaultImage:=Textures.T_Green_Check, DefaultDesiredSize:=vector2{X:=300.0, Y:=300.0}}
            CurrencyFrame :=texture_block{DefaultImage:=Textures.Currency.T_GoldSlot, DefaultDesiredSize:=vector2{X:=300.0, Y:=65.0}}
            CurrencyValue :=text_block{DefaultText:=StringToMessage("000000"), DefaultTextColor:=NamedColors.Black}
            CurrencyIcon :=texture_block{DefaultImage:=Textures.Currency.T_Gold, DefaultDesiredSize:=vector2{X:=45.0, Y:=45.0}}
            #Definition for the area taht holds the title, coin currency and "back" button
            StackBoxTitle :=stack_box:
                Orientation:=orientation.Horizontal
                Slots:=array:
                    stack_box_slot:
                        HorizontalAlignment:=horizontal_alignment.Fill
                        VerticalAlignment := vertical_alignment.Center
                        Widget:= overlay:
                            Slots:=array:
                                            
                                #Defines the first slot for the overlay widget
                                    overlay_slot:
                                        VerticalAlignment:=vertical_alignment.Center
                                        HorizontalAlignment:=horizontal_alignment.Left
                                        #Padding:=margin{Top:=50.0}
                                        #Defines that the widget that's going to be used within the first overlay slot is the background image previously defined
                                        Widget:= CurrencyFrame
                                #Defines the second slot for the overlay widget
                                    overlay_slot:
                                        VerticalAlignment:=vertical_alignment.Center
                                        HorizontalAlignment:=horizontal_alignment.Center
                                        Padding:=margin{Top:=50.0}
                                        #Defines that the widget that's going to be used within the second overlay slot is the vertical slot previously defined
                                        #along with the actual content
                                        Widget:=AssistTitle




            StackBoxHead :=stack_box:
                Orientation:=orientation.Horizontal
                Slots:=array:
                    stack_box_slot:
                        HorizontalAlignment:=horizontal_alignment.Fill
                        VerticalAlignment := vertical_alignment.Center
                        Padding:=margin{Left:=10.0, Right:=5.0, Top:=10.0}
                        Widget:= EnemiesButton
                    stack_box_slot:
                        HorizontalAlignment:=horizontal_alignment.Fill
                        VerticalAlignment := vertical_alignment.Center
                        Padding:=margin{Left:=5.0, Right:=5.0, Top:=10.0}
                        Widget:= BossesButton
                    stack_box_slot:
                        HorizontalAlignment:=horizontal_alignment.Fill
                        VerticalAlignment := vertical_alignment.Center
                        Padding:=margin{Left:=5.0, Right:=10.0, Top:=10.0}
                        Widget:= OthersButton

            StackBoxEnemiesTabs :=stack_box:
                Orientation:=orientation.Horizontal
                Slots:=array:
                    stack_box_slot:
                        #Sets the first tab, using an overlay and two layers; one for the BG iage and other for the text block
                        HorizontalAlignment:=horizontal_alignment.Center
                        VerticalAlignment:=vertical_alignment.Center
                        Widget:= overlay:
                            Slots:=array:
                                            
                                #Defines the first slot for the overlay widget
                                    overlay_slot:
                                        VerticalAlignment:=vertical_alignment.Center
                                        HorizontalAlignment:=horizontal_alignment.Center
                                        Padding:=margin{Top:=50.0}
                                        #Defines that the widget that's going to be used within the first overlay slot is the background image previously defined
                                        Widget:= AssistTab
                                #Defines the second slot for the overlay widget
                                    overlay_slot:
                                        VerticalAlignment:=vertical_alignment.Center
                                        HorizontalAlignment:=horizontal_alignment.Center
                                        Padding:=margin{Top:=50.0}
                                        #Defines that the widget that's going to be used within the second overlay slot is the vertical slot previously defined
                                        #along with the actual content
                                        Widget:=AssistTitle
                                
                    stack_box_slot:
                        #Sets the second tab, using an overlay and two layers; one for the BG iage and other for the text block
                        HorizontalAlignment:=horizontal_alignment.Center
                        VerticalAlignment:=vertical_alignment.Center
                        Widget:= overlay:
                            Slots:=array:
            
                                #Defines the first slot for the overlay widget
                                    overlay_slot:
                                        VerticalAlignment:=vertical_alignment.Center
                                        HorizontalAlignment:=horizontal_alignment.Center
                                        Padding:=margin{Top:=50.0}
                                        #Defines that the widget that's going to be used within the first overlay slot is the background image previously defined
                                        Widget:= ElimTab
                                #Defines the second slot for the overlay widget
                                    overlay_slot:
                                        VerticalAlignment:=vertical_alignment.Center
                                        HorizontalAlignment:=horizontal_alignment.Center
                                        Padding:=margin{Top:=50.0}
                                        #Defines that the widget that's going to be used within the second overlay slot is the vertical slot previously defined
                                        #along with the actual content
                                        Widget:=ElimTitle

            StackBoxBody :=stack_box:
                Orientation:=orientation.Vertical
                Slots:=array:

                    stack_box_slot:
                        HorizontalAlignment:=horizontal_alignment.Fill
                        VerticalAlignment := vertical_alignment.Center
                        Widget:= StackBoxTitle

                    stack_box_slot:
                        HorizontalAlignment:=horizontal_alignment.Fill
                        VerticalAlignment := vertical_alignment.Center
                        Widget:= StackBoxHead

                    stack_box_slot:
                        HorizontalAlignment:=horizontal_alignment.Center
                        VerticalAlignment := vertical_alignment.Center
                        Widget:= StackBoxEnemiesTabs
            
            StackBoxCurrencyContent :=stack_box:
                Orientation:=orientation.Horizontal
                Slots:=array:
                    stack_box_slot:
                        HorizontalAlignment:= horizontal_alignment.Center
                        VerticalAlignment:= vertical_alignment.Center
                        Padding:=margin{Left:=15.0}
                        Widget:= CurrencyIcon
                    stack_box_slot:
                        HorizontalAlignment:= horizontal_alignment.Center
                        VerticalAlignment:= vertical_alignment.Center
                        Widget:= CurrencyValue


            #Defines the canvas' parameteres
            Canvas:=canvas:
                Slots:=array:
                    #Defines the parameters for the canvas' slot
                    canvas_slot:
                        Anchors := anchors{Minimum := vector2{X := 0.0, Y := 0.0}, Maximum := vector2{X := 1.0, Y := 1.0}}
                        Alignment := vector2{X := 0.5, Y := 0.5}
                        #Defines that the widget that is going to be used within the canvas' slot is an overlay
                        Widget:=overlay:
                            Slots:=array:
                                
                                #Defines the first slot for the overlay widget
                                overlay_slot:
                                    VerticalAlignment:=vertical_alignment.Fill
                                    HorizontalAlignment:=horizontal_alignment.Fill #Padding:=margin{Left:=30.0, Bottom:=20.0}
                                    #Defines that the widget that's going to be used within the first overlay slot is the background image previously defined
                                    Widget:= ImageBackground
                                #Defines the second slot for the overlay widget
                                overlay_slot:
                                    VerticalAlignment:=vertical_alignment.Fill
                                    HorizontalAlignment:=horizontal_alignment.Center
                                    #Defines that the widget that's going to be used within the second overlay slot is the vertical slot previously defined
                                    #along with the actual content
                                    Widget:=StackBoxBody

                                            



            if(PlayerUI:=GetPlayerUI[Player]):
                PlayerUI.AddWidget(Canvas)

                
            