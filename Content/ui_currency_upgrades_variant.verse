<#
using { /Fortnite.com/UI }
using { /Fortnite.com/Devices }
using { /Verse.org/Simulation }
using { /Verse.org/Colors }
using { /UnrealEngine.com/Temporary/UI }
using { /UnrealEngine.com/Temporary/SpatialMath }

ui_currency_upgrades_variant := class(ui_common):    

    CreateCanvas(): void=
        Players:=GetPlayspace().GetPlayers()
        if(Player:=Players[0]):
         
    @editable
    Currency:currency_id = currency_id.Gold

    @editable
    TitleString:string = "Coin$ Rewards"

    @editable
    TabConfigs:[]currency_upgrade_ui_tab_config = array:
        currency_upgrade_ui_tab_config{Title:="ENEMIES", Upgrades:=array{currency_upgrade_id.Elimination, currency_upgrade_id.Assist}, MinTier:=option{0}, MaxTier:=option{4}}
        currency_upgrade_ui_tab_config{Title:="BOSSES", Upgrades:=array{currency_upgrade_id.Elimination, currency_upgrade_id.Assist}, MinTier:=option{5}, MaxTier:=option{9}}
        currency_upgrade_ui_tab_config{Title:="OTHERS", Upgrades:=array{currency_upgrade_id.Rescue}, MinTier:=option{0}, MaxTier:=option{0}}
        currency_upgrade_ui_tab_config{Title:="REST ZONE", Upgrades:=array{currency_upgrade_id.Rescue}, MinTier:=option{1}, MaxTier:=option{1}}

    @editable
    TierIconSize:vector2 = vector2{X:=143.0, Y:=143.0}

    @editable
    UpgradeString:string = "UPGRADE"
    
    @editable
    OnMenuCloseTrigger:trigger_device = trigger_device{}

    var MaybeCurrencyManager : ?clash_currency_manager = false

    PlayerTabSelectionEvent:event(currency_upgrade_tab_selection_payload) = event(currency_upgrade_tab_selection_payload){}

    PlayerUpgradeSelectionEvent:event(currency_upgrade_selection_payload) = event(currency_upgrade_selection_payload){}


    ShowUI(Player:player, CurrencyUpgradesManager:currency_upgrades_manager): void=
        spawn {AwaitShow(Player, CurrencyUpgradesManager)}

    
    AwaitShow(Player:player, CurrencyUpgradesManager:currency_upgrades_manager)<suspends>: void=
        if (Player, PlayerUI:=GetPlayerUI[Player]):
            set MaybeCurrencyManager = option{CurrencyUpgradesManager.CurrencyManager}

            SetPlayerInUI(Player, true)

            Logger.Print("Showing UI Currency Upgrades")

            WidgetRemovedEvent:event() = event(){} # So all buttons stop listening when widget removed

            UIData:=ConstructCanvas(Player, CurrencyUpgradesManager, WidgetRemovedEvent)

            PlayerUI.AddWidget(UIData.Base.Canvas, player_ui_slot{InputMode:=ui_input_mode.All}) #ZOrder:=UIZOrder, 

            AwaitShowing(Player, UIData, CurrencyUpgradesManager, WidgetRemovedEvent)

            Logger.Print("Hiding UI Currency Upgrades")

            WidgetRemovedEvent.Signal()
            PlayerUI.RemoveWidget(UIData.Base.Canvas)

            SetPlayerInUI(Player, false)
            OnMenuCloseTrigger.Trigger(Player)

    ConstructCanvas(Player:player, CurrencyUpgradesManager:currency_upgrades_manager, 
    WidgetRemovedEvent:event()): currency_upgrades_ui_data =
        Base:=ConstructBaseCanvas(Player, CurrencyUpgradesManager.CurrencyManager, TitleString, CloseString, WidgetRemovedEvent)

        Tabs:[]currency_upgrade_ui_tab_data = for (Index->TabConfig:TabConfigs):
            ConstructTab(Player, Index, TabConfig, WidgetRemovedEvent)
        
        for(Tab:Tabs):
            UpdateTabMaxOutCheck(Player, CurrencyUpgradesManager, Tab)
        
        TabsStackBox:=stack_box:
            Orientation:=orientation.Horizontal
            Slots:= for (Tab:Tabs):
                # Tab
                stack_box_slot:
                    VerticalAlignment:=vertical_alignment.Center
                    HorizontalAlignment:=horizontal_alignment.Fill
                    Padding:=margin{Left:=120.0, Right:=120.0}#{Left:=40.0, Right:=40.0}
                    
                    Widget:=overlay:
                        Slots:=array:
                            overlay_slot:
                                VerticalAlignment:=vertical_alignment.Center
                                HorizontalAlignment:=horizontal_alignment.Fill#Center                               
                                #Attempt at setting button width to "Fill"
                                Widget:=Tab.ButtonOverlay
                                #Apparently, this is the overlay that helds the button within the tab and the "highlight" color block

        AllContentStackBox:=stack_box:
            Orientation:=orientation.Vertical
            Slots:=array:
                # Tabs
                stack_box_slot:
                    VerticalAlignment:=vertical_alignment.Center
                    HorizontalAlignment:=horizontal_alignment.Fill#Center
                    Padding:=margin{Top:=5.0, Left:=0.0, Right:=0.0, Bottom:=5.0}#{Top:=5.0, Left:=200.0, Right:=100.0, Bottom:=5.0}
                    Widget:=TabsStackBox

        Base.MainStackBox.AddWidget(
            stack_box_slot:
                VerticalAlignment:=vertical_alignment.Fill
                HorizontalAlignment:=horizontal_alignment.Fill
                Padding:=margin{Left:=30.0, Right:=30.0}
                Distribution:=option{1.0}
                Widget:=AllContentStackBox
        )

        UIData:=currency_upgrades_ui_data:
            Base:=Base
            AllContentStackBox:=AllContentStackBox
            Tabs:=Tabs

        # UI Navigation with control is not created correctly if all tabs content is not created in advance, so we create all tab contents before showing
        for (Index->TabData:UIData.Tabs, TabConfig:=TabConfigs[Index]):
            ConstructAndAddContentTab(Player, UIData, TabConfig, TabData, CurrencyUpgradesManager, WidgetRemovedEvent)
            if (Content:=TabData.MaybeContentStackBox?):
                Content.SetVisibility(widget_visibility.Collapsed)

        UIData

    ConstructTab(Player:player, Index:int, TabConfig:currency_upgrade_ui_tab_config, WidgetRemovedEvent:event()): currency_upgrade_ui_tab_data =
        ButtonColorBlock:=color_block{DefaultColor:=NamedColors.Yellow}
        ButtonColorBlock.SetVisibility(widget_visibility.Hidden)

        #Apparent definition for "TabButton"
        TabButton:=button_regular{DefaultText:=StringToMessage(TabConfig.Title)}
        spawn {AwaitTabButtonSelection(TabButton, Index, WidgetRemovedEvent)}

        MaxoutCheck:=texture_block{DefaultImage:=Textures.T_Green_Check}
        MaxoutCheck.SetVisibility(widget_visibility.Collapsed)

        ButtonOverlay:=overlay:
            Slots:=array:
                # Color
                overlay_slot:
                    VerticalAlignment:=vertical_alignment.Fill
                    HorizontalAlignment:=horizontal_alignment.Fill
                    Padding:=margin{Top:=10.0, Left:=10.0, Right:=10.0, Bottom:=10.0}
                    Widget:=ButtonColorBlock
                # Button
                overlay_slot:
                    VerticalAlignment:=vertical_alignment.Center
                    HorizontalAlignment:=horizontal_alignment.Fill#Center
                    Widget:=TabButton

                # Max out check mark 
                overlay_slot:
                    VerticalAlignment:=vertical_alignment.Bottom
                    HorizontalAlignment:=horizontal_alignment.Right
                    Widget:=MaxoutCheck

        currency_upgrade_ui_tab_data:
            ButtonOverlay:=ButtonOverlay
            ButtonColorBlock:=ButtonColorBlock

            MaxoutCheck:= MaxoutCheck

            MaybeContentStackBox:=false
            UpgradeCards:=map{}

            TabConfig := TabConfig

    UnselectTab(Index:int, UIData:currency_upgrades_ui_data): void =
        if (TabData:=UIData.Tabs[Index]):
            TabData.ButtonColorBlock.SetVisibility(widget_visibility.Collapsed)

            if (Content:=TabData.MaybeContentStackBox?):
                Content.SetVisibility(widget_visibility.Collapsed)

    SelectTab(Player:player, Index:int, UIData:currency_upgrades_ui_data, CurrencyUpgradesManager:currency_upgrades_manager, WidgetRemovedEvent:event()): void =
        if:
            TabConfig:=TabConfigs[Index]
            TabData:=UIData.Tabs[Index]
        then:
            TabData.ButtonColorBlock.SetVisibility(widget_visibility.Visible)

            if (Content:=TabData.MaybeContentStackBox?):
                Content.SetVisibility(widget_visibility.Visible)
            else:
                ConstructAndAddContentTab(Player, UIData, TabConfig, TabData, CurrencyUpgradesManager, WidgetRemovedEvent)

    ConstructAndAddContentTab(Player:player, UIData:currency_upgrades_ui_data, TabConfig:currency_upgrade_ui_tab_config, TabData:currency_upgrade_ui_tab_data, CurrencyUpgradesManager:currency_upgrades_manager, WidgetRemovedEvent:event()): void =
        var UpgradeCards:[currency_upgrade_id][]currency_upgrade_ui_card_data = map{}
       

        UpgradeTitles:=stack_box:
            Orientation:=orientation.Horizontal
            Slots:= for (UpgradeId : TabConfig.Upgrades, 
            Definition:=CurrencyUpgradesManager.CurrencyUpgradeDefinitions.FindCurrencyUpgradeDefinition[Currency, UpgradeId, ?MinTier:=TabConfig.MinTier, ?MaxTier:=TabConfig.MaxTier]):
                var TabTitle : string = "By {Definition.DisplayName}"
                if(TabConfig.Title = "REST ZONE"):
                    set TabTitle = "Max Coin$ rate by {Definition.DisplayName}"
                # Upgrade Title
                stack_box_slot:
                    VerticalAlignment:=vertical_alignment.Fill
                    HorizontalAlignment:=horizontal_alignment.Center
                    Distribution:=option{1.0}
                    Widget:=overlay:
                        Slots:=array:
                            overlay_slot:
                                VerticalAlignment:=vertical_alignment.Fill
                                HorizontalAlignment:=horizontal_alignment.Fill
                                Widget:=texture_block{DefaultImage:=Textures.General.T_BlueTab, DefaultDesiredSize:=vector2{X:=359.0, Y:=56.0}}
                            overlay_slot:
                                VerticalAlignment:=vertical_alignment.Center
                                HorizontalAlignment:=horizontal_alignment.Center
                                Padding:=margin{Left:=30.0, Right:=30.0}
                                Widget:=text_block{DefaultText:=StringToMessage(TabTitle), DefaultTextColor:=NamedColors.White}

        Content:=stack_box:
            Orientation:=orientation.Vertical
            Slots:=array:
                # Upgrade Titles
                stack_box_slot:
                    VerticalAlignment:=vertical_alignment.Center
                    HorizontalAlignment:=horizontal_alignment.Fill
                    Padding:=margin{Left:=145.0}
                    Widget:=UpgradeTitles

        var IndexTier:int = 0
        
        for (RowUpgradeId : TabConfig.Upgrades, 
        Definition:=CurrencyUpgradesManager.CurrencyUpgradeDefinitions.FindCurrencyUpgradeDefinition[Currency, RowUpgradeId, ?MinTier:=TabConfig.MinTier, ?MaxTier:=TabConfig.MaxTier]):
            var MaxIndexTier : int =(Definition.Tiers.Length-1) 
            var IconTier : int = 0
            if (MinIndex:=TabConfig.MinTier?, MaxIndex:=TabConfig.MaxTier?):
                set IconTier = MinIndex
            for (CurrentTier:=IndexTier..MaxIndexTier):
                RealTier:=IconTier + CurrentTier
                Content.AddWidget(
                    TierStackBox:=stack_box:
                        Orientation:=orientation.Horizontal
                        Slots:=array:
                            # Tier Icon
                            stack_box_slot:
                                HorizontalAlignment:=horizontal_alignment.Fill
                                VerticalAlignment:=vertical_alignment.Fill
                                Widget:=overlay:
                                    Slots:=array:
                                        overlay_slot:
                                            VerticalAlignment:=vertical_alignment.Fill
                                            HorizontalAlignment:=horizontal_alignment.Fill
                                            Widget:=texture_block{DefaultImage:=Textures.General.T_BlueBorder, DefaultDesiredSize:=vector2{X:=100.0, Y:=100.0}}
                                        overlay_slot:
                                            VerticalAlignment:=vertical_alignment.Center
                                            HorizontalAlignment:=horizontal_alignment.Center
                                            Padding:=margin{Top:=3.0, Left:=1.0, Right:=1.0, Bottom:=3.0}
                                            Widget:=texture_block{DefaultImage:=FindTierIcon(RowUpgradeId, RealTier), DefaultDesiredSize:=TierIconSize}

                    for (UpgradeId:TabConfig.Upgrades):
                        TierStackBox.AddWidget(
                            UpgradeOverlay:=overlay:
                                Slots:=array:
                                    overlay_slot:
                                        VerticalAlignment:=vertical_alignment.Fill
                                        HorizontalAlignment:=horizontal_alignment.Fill
                                        Widget:=texture_block{DefaultImage:=Textures.General.T_BlueBorder, DefaultDesiredSize:=vector2{X:=200.0, Y:=100.0}}

                            CardData:=ConstructUpgradeCard(Player, CurrencyUpgradesManager, UpgradeId, RealTier, WidgetRemovedEvent)

                            UpgradeOverlay.AddWidget(
                                # Upgrade Card
                                overlay_slot:
                                    VerticalAlignment:=vertical_alignment.Fill
                                    HorizontalAlignment:=horizontal_alignment.Fill
                                    Widget:=CardData.StackBox
                            )

                            var Cards:[]currency_upgrade_ui_card_data = array{}
                            if (CurrentCards:=UpgradeCards[UpgradeId]):
                                set Cards = CurrentCards
                            set Cards += array{CardData}
                            if (set UpgradeCards[UpgradeId] = Cards) {}
                        
                            # Upgrade Container
                            stack_box_slot:
                                HorizontalAlignment:=horizontal_alignment.Fill
                                VerticalAlignment:=vertical_alignment.Fill
                                Distribution:=option{1.0}
                                Widget:=UpgradeOverlay
                        )

                    # Tier
                    stack_box_slot:
                        VerticalAlignment:=vertical_alignment.Center
                        HorizontalAlignment:=horizontal_alignment.Fill
                        Widget:=TierStackBox
                )

                set IndexTier +=1

        set TabData.MaybeContentStackBox = option{Content}
        set TabData.UpgradeCards = UpgradeCards

        # Add content to widget display
        UIData.AllContentStackBox.AddWidget(
            # Tab Content
            stack_box_slot:
                VerticalAlignment:=vertical_alignment.Fill
                HorizontalAlignment:=horizontal_alignment.Center
                Padding:=margin{Top:=5.0}
                Distribution:=option{1.0}
                Widget:=Content
        )

    ConstructUpgradeCard(Player:player, CurrencyUpgradesManager:currency_upgrades_manager, 
    Upgrade:currency_upgrade_id, Tier:int, WidgetRemovedEvent:event()): currency_upgrade_ui_card_data =
        LevelText:=text_block{DefaultText:=StringToMessage(""), DefaultTextColor:=NamedColors.Black}
        StatText:=text_block{DefaultText:=StringToMessage(""), DefaultTextColor:=NamedColors.White}
        NextLevelText:=text_block{DefaultText:=StringToMessage(""), DefaultTextColor:=NamedColors.White}
        
        SelectUpgradeButton:button_loud = button_loud{}
        spawn {AwaitUpgradeButtonSelection(SelectUpgradeButton, Upgrade, Tier, WidgetRemovedEvent)}

        MaxoutButton:=button_quiet{}
        MaxoutButton.SetVisibility(widget_visibility.Collapsed)
        MaxoutButton.SetEnabled(false)
        spawn {AwaitUpgradeButtonSelection(MaxoutButton, Upgrade, Tier, WidgetRemovedEvent)}

        MaxoutCheck:=texture_block{DefaultImage:=Textures.T_Green_Check}
        MaxoutCheck.SetVisibility(widget_visibility.Collapsed)
        MaxoutCheck.SetEnabled(false)

        MaxoutOverlay:= overlay:
            Slots:=array:
                overlay_slot:
                    HorizontalAlignment:=horizontal_alignment.Right
                    VerticalAlignment:=vertical_alignment.Bottom
                    Padding:=margin{Bottom:=-10.0, Right:=-10.0}
                    Widget:=MaxoutCheck

                overlay_slot:
                    HorizontalAlignment:=horizontal_alignment.Fill
                    VerticalAlignment:=vertical_alignment.Fill
                    Widget:=MaxoutButton

        StackBox:=stack_box:
            Orientation:=orientation.Horizontal
            Slots:=array:
                # Stats
                stack_box_slot:
                    HorizontalAlignment:=horizontal_alignment.Fill
                    VerticalAlignment:=vertical_alignment.Center
                    Padding:=margin{Left:=20.0, Right:=20.0}
                    Distribution:=option{1.0}
                    Widget:=stack_box:
                        Orientation:=orientation.Vertical
                        Slots:=array:
                            # Current
                            stack_box_slot:
                                VerticalAlignment:=vertical_alignment.Center
                                HorizontalAlignment:=horizontal_alignment.Left
                                Widget:=overlay:
                                    Slots:=array:
                                        # Level
                                        overlay_slot:
                                            VerticalAlignment:=vertical_alignment.Top
                                            HorizontalAlignment:=horizontal_alignment.Left
                                            Widget:=overlay:
                                                Slots:=array:
                                                    # Circle Background
                                                    overlay_slot:
                                                        VerticalAlignment:=vertical_alignment.Fill
                                                        HorizontalAlignment:=horizontal_alignment.Fill
                                                        Widget:=texture_block{DefaultImage:=Textures.General.T_YellowCircle, DefaultDesiredSize:=vector2{X:=100.0, Y:=100.0}}
                                                    # Level Text
                                                    overlay_slot:
                                                        HorizontalAlignment:=horizontal_alignment.Center
                                                        VerticalAlignment:=vertical_alignment.Center
                                                        Padding:=margin{Top:=10.0, Left:=10.0, Right:=10.0, Bottom:=10.0}
                                                        Widget:=LevelText
                                        
                                        # Stat
                                        overlay_slot:
                                            VerticalAlignment:=vertical_alignment.Center
                                            HorizontalAlignment:=horizontal_alignment.Fill
                                            Padding:=margin{Left:=120.0, Right:=20.0}
                                            Widget:=overlay:
                                                Slots:=array:
                                                    # Stat Text
                                                    overlay_slot:
                                                        HorizontalAlignment:=horizontal_alignment.Center
                                                        VerticalAlignment:=vertical_alignment.Center
                                                        Widget:=StatText

                            # Next
                            stack_box_slot:
                                VerticalAlignment:=vertical_alignment.Center
                                HorizontalAlignment:=horizontal_alignment.Left
                                Widget:=NextLevelText

                # Button
                stack_box_slot:
                    HorizontalAlignment:=horizontal_alignment.Center
                    VerticalAlignment:=vertical_alignment.Center
                    Padding:=margin{Left:=10.0, Right:=20.0}
                    Widget:=overlay:
                        Slots:=array:
                            overlay_slot:
                                HorizontalAlignment:=horizontal_alignment.Fill
                                VerticalAlignment:=vertical_alignment.Fill
                                Widget:=MaxoutOverlay
                            overlay_slot:
                                HorizontalAlignment:=horizontal_alignment.Fill
                                VerticalAlignment:=vertical_alignment.Fill
                                Widget:=SelectUpgradeButton
                            

        CardData:=currency_upgrade_ui_card_data:
            Tier:=Tier
            StackBox:=StackBox

            LevelText:=LevelText
            StatText:=StatText
            NextLevelText:=NextLevelText
            Button:=SelectUpgradeButton
            MaxoutButton:=MaxoutButton
            MaxoutCheckMark:=MaxoutCheck

        if:
            TierDefinition:=CurrencyUpgradesManager.CurrencyUpgradeDefinitions.FindCurrencyUpgradeTierDefinition[Currency, Upgrade, Tier]
            CurrentLevel := CurrencyUpgradesManager.GetPlayerClampedLevelInCurrencyUpgrade(TierDefinition, Player, Currency, Upgrade, Tier)
        then:
            UpdateDisplayUpgradeCard(Player, CurrencyUpgradesManager, TierDefinition, Upgrade, CardData, CurrentLevel)

        return CardData

    UpdateDisplayUpgradeCard(Player:player, CurrencyUpgradesManager:currency_upgrades_manager, TierDefinition:currency_upgrade_tier_definition,
    Upgrade:currency_upgrade_id, CardData:currency_upgrade_ui_card_data, CurrentLevelIndex:int): void =

        NextLevelIndex := CurrentLevelIndex + 1

        CardData.LevelText.SetText(StringToMessage("Lvl {CurrentLevelIndex+1}"))
        CardData.StatText.SetText(StringToMessage("Get ${ToNotationString(TierDefinition.GetValueInClampedLevel(CurrentLevelIndex))}"))

        if (NextLevelStat:=TierDefinition.HasLevel[NextLevelIndex]):
            CardData.NextLevelText.SetText(StringToMessage("Next Lvl: Get ${ToNotationString(NextLevelStat)}"))
            CardData.NextLevelText.SetVisibility(widget_visibility.Visible)

            if:
                CurrencyManager:=MaybeCurrencyManager?
                NextLevelCost := CurrencyUpgradesManager.GetCurrencyUpgradeCostForLevel[Player, TierDefinition.IncrementCost.Currency, Upgrade, CardData.Tier, NextLevelIndex]
            then:
                CurrentPlayerCurrency:=CurrencyManager.GetPlayerCurrency(Player, NextLevelCost.Currency)
                if (CurrentPlayerCurrency >= NextLevelCost.Value):
                    CardData.Button.SetText(StringToMessage("{UpgradeString}\n${ToNotationString(NextLevelCost.Value)}"))
                    CardData.Button.SetVisibility(widget_visibility.Visible)
                    CardData.Button.SetEnabled(true)
                    CardData.MaxoutButton.SetVisibility(widget_visibility.Collapsed)
                    CardData.MaxoutButton.SetEnabled(false)
                else:
                    CardData.MaxoutButton.SetText(StringToMessage("{UpgradeString}\n${ToNotationString(NextLevelCost.Value)}"))
                    CardData.MaxoutButton.SetVisibility(widget_visibility.Visible)
                    CardData.MaxoutButton.SetEnabled(true) 
                    CardData.Button.SetVisibility(widget_visibility.Collapsed)
                    CardData.Button.SetEnabled(false)       

        else:
            CardData.NextLevelText.SetVisibility(widget_visibility.Collapsed)            
            CardData.MaxoutButton.SetText(StringToMessage("Max out"))            
            CardData.MaxoutCheckMark.SetEnabled(true)
            CardData.MaxoutCheckMark.SetVisibility(widget_visibility.Visible)
            CardData.MaxoutButton.SetEnabled(true)
            CardData.MaxoutButton.SetVisibility(widget_visibility.Visible)
            CardData.Button.SetEnabled(false)
            CardData.Button.SetVisibility(widget_visibility.Collapsed)
            

    AwaitShowing(Player:player, UIData:currency_upgrades_ui_data, 
    CurrencyUpgradesManager:currency_upgrades_manager, WidgetRemovedEvent:event())<suspends>: void =
        var CurrentTabIndex:int = 0
        SelectTab(Player, CurrentTabIndex, UIData, CurrencyUpgradesManager, WidgetRemovedEvent)

        race:
            #AwaitHideEvent(Player)
            
            #loop:
                #AwaitCurrencyChanged(Player, CurrencyUpgradesManager.CurrencyManager, UIData.Base)

            #loop:
                #AwaitShowNoMoneyMessage(Player, UIData.Base)

            loop:
                TabSelection := PlayerTabSelectionEvent.Await()
                if (TabSelection.Player = Player):
                    UnselectTab(CurrentTabIndex, UIData)

                    set CurrentTabIndex = TabSelection.Index
                    SelectTab(Player, CurrentTabIndex, UIData, CurrencyUpgradesManager, WidgetRemovedEvent)

            #loop:
                #UpgradeSelection := PlayerUpgradeSelectionEvent.Await()
                #if (UpgradeSelection.Player = Player):
                    #bSuccess := CurrencyUpgradesManager.TryIncrementUpgrade(Player, Currency, UpgradeSelection.Id, UpgradeSelection.Tier)
                    #layInteractionSound(bSuccess, Player)
                    #if(not bSuccess?):
                        #NoCurrencyEvent.Signal(Player)

            loop:
                CurrencyUpgradeChangedData := CurrencyUpgradesManager.CurrencyUpgradeChangedEvent.Await()
                if:
                    CurrencyUpgradeChangedData.Player = Player
                    CurrencyUpgradeChangedData.Currency = Currency
                then:
                    for (Tab:UIData.Tabs):
                        for (KeyUpgrade->Cards:Tab.UpgradeCards):
                            for (Card:Cards):
                                if (TierDefinition:=CurrencyUpgradesManager.CurrencyUpgradeDefinitions.FindCurrencyUpgradeTierDefinition[Currency, KeyUpgrade, Card.Tier]):
                                    if:
                                        KeyUpgrade = CurrencyUpgradeChangedData.UpgradeId
                                        Card.Tier = CurrencyUpgradeChangedData.Tier
                                    then:
                                        UpdateDisplayUpgradeCard(Player, CurrencyUpgradesManager, TierDefinition, KeyUpgrade, Card, CurrencyUpgradeChangedData.NewNumIncrements)
                                        UpdateTabMaxOutCheck(Player, CurrencyUpgradesManager, Tab)
                                    else:
                                        CurrentLevel:= CurrencyUpgradesManager.GetPlayerClampedLevelInCurrencyUpgrade(TierDefinition, Player, Currency, KeyUpgrade, Card.Tier)
                                        UpdateDisplayUpgradeCard(Player, CurrencyUpgradesManager, TierDefinition, KeyUpgrade, Card, CurrentLevel)

    AwaitTabButtonSelection(Button:text_button_base, Index:int, WidgetRemovedEvent:event())<suspends>: void =
        race:
            block:
                WidgetRemovedEvent.Await()
                return
            loop:
                WidgetMessage := Button.OnClick().Await()
                PlayerTabSelectionEvent.Signal(currency_upgrade_tab_selection_payload{Player:=WidgetMessage.Player, Index:=Index})

    AwaitUpgradeButtonSelection(Button:text_button_base, Upgrade:currency_upgrade_id, Tier:int, WidgetRemovedEvent:event())<suspends>: void =
        race:
            block:
                WidgetRemovedEvent.Await()
                return
            loop:
                WidgetMessage := Button.OnClick().Await()
                PlayerUpgradeSelectionEvent.Signal(currency_upgrade_selection_payload{Player:=WidgetMessage.Player, Id:=Upgrade, Tier:=Tier})

    UpdateTabMaxOutCheck(Player:player, CurrencyUpgradesManager:currency_upgrades_manager, TabData:currency_upgrade_ui_tab_data):void=
        TabConfig := TabData.TabConfig
        for(Upgrade:TabConfig.Upgrades):
            if(MinTier:= TabConfig.MinTier?, MaxTier:=TabConfig.MaxTier?):
                for(Tier:=MinTier..MaxTier):
                    if(not CurrencyUpgradesManager.IsUpgradeMaxedOut(Player, Currency, Upgrade, Tier)?):
                        return

        TabData.MaxoutCheck.SetVisibility(widget_visibility.Visible)
        #>