using { /Fortnite.com/UI }
using { /Fortnite.com/Characters }
using { /Verse.org/Simulation }
using { /Verse.org/Colors }
using { /UnrealEngine.com/Temporary/UI }
using { /UnrealEngine.com/Temporary/SpatialMath }

ui_hud_currency_widget_data:=struct:
    Canvas:canvas
    GoldTextBlock:text_block

ui_hud_currency:=class():

    @editable
	CurrenciesPosition:vector2 = vector2{X:=-10.0,Y:=-260.0}

	@editable
	CurrencyTextColor:color = NamedColors.Black

	@editable
	CurrencySlotSize:vector2 = vector2{X:=300.0,Y:=65.0}

	@editable
	CurrencyIconSize:vector2 = vector2{X:=45.0,Y:=45.0}

    #Events

    CurrencyHandlerEvent:event(tuple(player, float, int))=event(tuple(player, float, int)){}
    CloseWidgetEvent:event(player)=event(player){}
                                                     
    AddWidgetToPlayer(Player:player, PlayerCurrency:float):void=
        CreateWidget(Player, PlayerCurrency)
        # Print("Creating Player Currency Widget")

    CreateWidget<protected>(Player:player, PlayerCurrency:float):ui_hud_currency_widget_data=

        Canvas:=canvas{}

        # Currencies
        CurrenciesStackBox:stack_box = stack_box{Orientation:=orientation.Vertical}
            var CurrencyTextBlocks:weak_map(currency_id, text_block) = map{}


                TextBlock:=text_block:
                    DefaultText:=StringToMessage("{ToNotationString(PlayerCurrency)}")
                    DefaultTextColor:=CurrencyTextColor

                CurrenciesStackBox.AddWidget(
                    stack_box_slot:
                        HorizontalAlignment:=horizontal_alignment.Fill
                        VerticalAlignment:=vertical_alignment.Fill
                        Padding:=margin{Bottom:=5.0}
                        Widget:=overlay:
                            Slots:=array:
                                # Background
                                overlay_slot:
                                    HorizontalAlignment:=horizontal_alignment.Fill
                                    VerticalAlignment:=vertical_alignment.Fill
                                    Widget:=texture_block:
                                        DefaultImage:=Textures.Currency.T_GoldSlot
                                        DefaultDesiredSize:=CurrencySlotSize
                                # Icon
                                overlay_slot:
                                    HorizontalAlignment:=horizontal_alignment.Left
                                    VerticalAlignment:=vertical_alignment.Center
                                    Padding:=margin{Left:=15.0}
                                    Widget:=texture_block:
                                        DefaultImage:=Textures.Currency.T_Gold
                                        DefaultDesiredSize:=CurrencyIconSize
                                # Text
                                overlay_slot:
                                    HorizontalAlignment:=horizontal_alignment.Fill
                                    VerticalAlignment:=vertical_alignment.Center
                                    Padding:=margin{Left:=90.0, Right:=10.0}
                                    Widget:=TextBlock
                )

        CurrencyWidgetData:=ui_hud_currency_widget_data{
            Canvas:=Canvas
            GoldTextBlock:=TextBlock
        }

        Canvas.AddWidget(canvas_slot{ 
            Anchors:=anchors{Minimum:=vector2{X:=1.0,Y:=1.0},Maximum:=vector2{X:=1.0,Y:=1.0}}
            Offsets:=margin{Top:=CurrenciesPosition.Y, Left:=CurrenciesPosition.X, Right:=0.0, Bottom:=0.0}
            Alignment:=vector2{X:=1.0, Y:=1.0}
            SizeToContent:=true
            Widget:=CurrenciesStackBox
            }     
            )

        if(PlayerUI:=GetPlayerUI[Player]):
            PlayerUI.AddWidget(Canvas)

        spawn{CurrencyEventHandler(Player, CurrencyWidgetData)}

        return CurrencyWidgetData

    CurrencyEventHandler<protected>(Player:player, CurrencyWidgetData:ui_hud_currency_widget_data)<suspends>:void=
        CloseWidgetInstanceEvent:=event(){}
        race:
            CloseWidgetInstanceEvent.Await()
            loop:
                EventPlayer:=CloseWidgetEvent.Await()
                if(EventPlayer = Player):
                    CloseWidgetInstanceEvent.Signal()
            loop:
                Data:=CurrencyHandlerEvent.Await()
                if(Data(0) = Player):
                    AmountOfGold:=Data(1)
                    StringAmountOfGold:=StringToMessage(ToNotationString(AmountOfGold))
                    CurrencyWidgetData.GoldTextBlock.SetText(StringAmountOfGold)
            # loop:
            #     if(FortCharacter:=fort_character[Player]):
            #         FortCharacter.EliminatedEvent().Await()
            #         CloseWidgetInstanceEvent.Signal()

                    
