using { /Fortnite.com/Devices }
using { /Fortnite.com/UI}
using { /Verse.org/Simulation }
using { /Verse.org/Colors }
using { /Verse.org/Colors/NamedColors }
using { /UnrealEngine.com/Temporary/UI }
using { /UnrealEngine.com/Temporary/SpatialMath }

ui_hud_match_marker := class(clash_ui):#class(ui_common)

	#Red #db242aFF
	#Gray #c5c6c8FF
	#Blue #417af4FF

	var RoomRedSide1: texture_block = texture_block{DefaultImage:=Textures.General.T_GenericSquare, DefaultDesiredSize:=vector2{X:=18.0, Y:=18.0}, DefaultTint:=MakeColorFromHex("#c5c6c8FF")}
	var RoomRedSide2: texture_block = texture_block{DefaultImage:=Textures.General.T_GenericSquare, DefaultDesiredSize:=vector2{X:=18.0, Y:=18.0}, DefaultTint:=MakeColorFromHex("#c5c6c8FF")}
	var RoomRedSide3: texture_block = texture_block{DefaultImage:=Textures.General.T_GenericSquare, DefaultDesiredSize:=vector2{X:=18.0, Y:=18.0}, DefaultTint:=MakeColorFromHex("#c5c6c8FF")}
	var RoomRedSide4: texture_block = texture_block{DefaultImage:=Textures.General.T_GenericSquare, DefaultDesiredSize:=vector2{X:=18.0, Y:=18.0}, DefaultTint:=MakeColorFromHex("#c5c6c8FF")}

	var RoomArena: texture_block = texture_block{DefaultImage:=Textures.General.T_GenericSquare, DefaultDesiredSize:=vector2{X:=42.0, Y:=22.0}, DefaultTint:=MakeColorFromHex("#c5c6c8FF")}

	var RoomBlueSide1: texture_block = texture_block{DefaultImage:=Textures.General.T_GenericSquare, DefaultDesiredSize:=vector2{X:=18.0, Y:=18.0}, DefaultTint:=MakeColorFromHex("#c5c6c8FF")}
	var RoomBlueSide2: texture_block = texture_block{DefaultImage:=Textures.General.T_GenericSquare, DefaultDesiredSize:=vector2{X:=18.0, Y:=18.0}, DefaultTint:=MakeColorFromHex("#c5c6c8FF")}
	var RoomBlueSide3: texture_block = texture_block{DefaultImage:=Textures.General.T_GenericSquare, DefaultDesiredSize:=vector2{X:=18.0, Y:=18.0}, DefaultTint:=MakeColorFromHex("#c5c6c8FF")}
	var RoomBlueSide4: texture_block = texture_block{DefaultImage:=Textures.General.T_GenericSquare, DefaultDesiredSize:=vector2{X:=18.0, Y:=18.0}, DefaultTint:=MakeColorFromHex("#c5c6c8FF")}

	var RoomName : text_block = text_block{DefaultTextColor:=MakeColorFromHex("#FFFFFFFF"), DefaultShadowColor:=MakeColorFromHex("#000000FF"), DefaultShadowOffset:=option{vector2{X:=3.0, Y:=3.0}}}

	var RedBed1 : texture_block = texture_block{DefaultImage:=Textures.General.T_GenArrow, DefaultDesiredSize:=vector2{X:=48.0, Y:=30.0}, DefaultTint:=MakeColorFromHex("#741a1dFF")}
	var RedBed2 : texture_block = texture_block{DefaultImage:=Textures.General.T_GenArrow, DefaultDesiredSize:=vector2{X:=48.0, Y:=30.0}, DefaultTint:=MakeColorFromHex("#741a1dFF")}
	var RedBed3 : texture_block = texture_block{DefaultImage:=Textures.General.T_GenArrow, DefaultDesiredSize:=vector2{X:=48.0, Y:=30.0}, DefaultTint:=MakeColorFromHex("#741a1dFF")}
	var RedBed4 : texture_block = texture_block{DefaultImage:=Textures.General.T_GenArrow, DefaultDesiredSize:=vector2{X:=48.0, Y:=30.0}, DefaultTint:=MakeColorFromHex("#741a1dFF")}

	var BlueBed1 : texture_block = texture_block{DefaultImage:=Textures.General.T_GenArrowLeft, DefaultDesiredSize:=vector2{X:=48.0, Y:=30.0}, DefaultTint:=MakeColorFromHex("#1e3a76FF")}
	var BlueBed2 : texture_block = texture_block{DefaultImage:=Textures.General.T_GenArrowLeft, DefaultDesiredSize:=vector2{X:=48.0, Y:=30.0}, DefaultTint:=MakeColorFromHex("#1e3a76FF")}
	var BlueBed3 : texture_block = texture_block{DefaultImage:=Textures.General.T_GenArrowLeft, DefaultDesiredSize:=vector2{X:=48.0, Y:=30.0}, DefaultTint:=MakeColorFromHex("#1e3a76FF")}
	var BlueBed4 : texture_block = texture_block{DefaultImage:=Textures.General.T_GenArrowLeft, DefaultDesiredSize:=vector2{X:=48.0, Y:=30.0}, DefaultTint:=MakeColorFromHex("#1e3a76FF")}

	var RedBedStackBox:stack_box = stack_box{Orientation:=orientation.Horizontal}

	var BlueBedStackBox:stack_box = stack_box{Orientation:=orientation.Horizontal}

	@editable
	var AmountOfBedsToShow:int = 3

	#------------------------------------------------------------------------------------------------------------
	CreateUI<override>():canvas_slot =
		set UICanvas = canvas_slot:
			Anchors := anchors{Minimum := vector2{X := 0.2, Y := 0.04}, Maximum := vector2{X := 0.8, Y := 0.1}} # Anchors := anchors{Minimum := vector2{X := 0.227, Y := 0.0}, ...
						Alignment := vector2{X := 0.5, Y := 0.7} # Alignment := vector2{X := 0.5, Y := 0.5} ...
						Widget:=overlay:
							Slots:=array:
								overlay_slot:
									VerticalAlignment:=vertical_alignment.Top
									HorizontalAlignment:=horizontal_alignment.Center
									Padding:=margin{Left:=0.0, Top:=0.0, Right:=0.0, Bottom:=0.0}
									#Screen background
									Widget:=texture_block{DefaultImage:=Textures.General.T_GenSquTrans, DefaultDesiredSize:=vector2{X:=1810.0, Y:=170.0}}
							
								overlay_slot:
									VerticalAlignment:=vertical_alignment.Center
									HorizontalAlignment:=horizontal_alignment.Center
									Padding:=margin{Left:=0.0, Top:=0.0, Right:=0.0, Bottom:=0.0}
									#Screen background
									Widget:=CreateMatchMaker()

		return UICanvas

	#------------------------------------------------------------------------------------------------------------
	CreateMatchMaker():overlay =

		HudOverlay := overlay:
			Slots:=array:
				#Makeshift size box
				overlay_slot:
					VerticalAlignment:=vertical_alignment.Center
					HorizontalAlignment:=horizontal_alignment.Fill
					#Distribution:=option{0.5}
					Padding:=margin{Left:=0.0, Top:=0.0, Right:=0.0, Bottom:=0.0}
					Widget:=texture_block{DefaultImage:=Textures.General.T_GenSquTrans, DefaultDesiredSize:=vector2{X:=552.0, Y:=115.0}, DefaultTint:=MakeColorFromHex("#FF000066")}

				#Main layout
				overlay_slot:
					VerticalAlignment:=vertical_alignment.Top
					HorizontalAlignment:=horizontal_alignment.Center
					#Distribution:=option{0.5}
					Padding:=margin{Left:=0.0, Top:=0.0, Right:=0.0, Bottom:=0.0}
					Widget:=stack_box:
						Orientation:=orientation.Vertical
						Slots:=array:

							#Actual progress module
							stack_box_slot:
								VerticalAlignment:=vertical_alignment.Center
								HorizontalAlignment:=horizontal_alignment.Center
								Distribution:=option{1.0}
								Padding:=margin{Left:=0.0, Top:=0.0, Right:=0.0, Bottom:=0.0}
								Widget:=stack_box:
									Orientation:=orientation.Vertical
									Slots:=array:

										#Match progress module
										stack_box_slot:
											VerticalAlignment:=vertical_alignment.Center
											HorizontalAlignment:=horizontal_alignment.Center
											#Distribution:=option{1.0}
											Padding:=margin{Left:=0.0, Top:=0.0, Right:=0.0, Bottom:=0.0}
											Widget:=stack_box:
												Orientation:=orientation.Horizontal
												Slots:=array:
													
													#Red team beds
													stack_box_slot:
														VerticalAlignment:=vertical_alignment.Center
														HorizontalAlignment:=horizontal_alignment.Right
														#Distribution:=option{1.0}
														Padding:=margin{Left:=0.0, Top:=0.0, Right:=0.0, Bottom:=0.0}
														Widget:=overlay:
															Slots:=array:
																
																#Makeshift size box for first team beds
																overlay_slot:
																	VerticalAlignment:=vertical_alignment.Center
																	HorizontalAlignment:=horizontal_alignment.Center
																	#Distribution:=option{1.0}
																	Padding:=margin{Left:=0.0, Top:=0.0, Right:=0.0, Bottom:=0.0}
																	Widget:=texture_block{DefaultImage:=Textures.General.T_GenSquTrans, DefaultDesiredSize:=vector2{X:=315.0, Y:=65.0}, DefaultTint:=MakeColorFromHex("#0000FFFF")}

																#Actual content for first team beds
																overlay_slot:
																	VerticalAlignment:=vertical_alignment.Center
																	HorizontalAlignment:=horizontal_alignment.Right
																	#Distribution:=option{1.0}
																	Padding:=margin{Left:=0.0, Top:=0.0, Right:=0.0, Bottom:=10.0}
																	Widget:=ConstructRedStackBox()

													#Header - room number
													stack_box_slot:
														VerticalAlignment:=vertical_alignment.Center
														HorizontalAlignment:=horizontal_alignment.Center
														#Distribution:=option{1.0}
														Padding:=margin{Left:=35.0, Top:=0.0, Right:=35.0, Bottom:=10.0}
														Widget:= RoomName
													
													#Blue team beds
													stack_box_slot:
														VerticalAlignment:=vertical_alignment.Center
														HorizontalAlignment:=horizontal_alignment.Left
														#Distribution:=option{1.0}
														Padding:=margin{Left:=0.0, Top:=0.0, Right:=0.0, Bottom:=0.0}
														Widget:=overlay:
															Slots:=array:
																
																#Makeshift size box for first team beds
																overlay_slot:
																	VerticalAlignment:=vertical_alignment.Center
																	HorizontalAlignment:=horizontal_alignment.Center
																	#Distribution:=option{1.0}
																	Padding:=margin{Left:=0.0, Top:=0.0, Right:=0.0, Bottom:=0.0}
																	Widget:=texture_block{DefaultImage:=Textures.General.T_GenSquTrans, DefaultDesiredSize:=vector2{X:=315.0, Y:=65.0}, DefaultTint:=MakeColorFromHex("#0000FFFF")}

																#Actual content for first team beds
																overlay_slot:
																	VerticalAlignment:=vertical_alignment.Center
																	HorizontalAlignment:=horizontal_alignment.Left
																	#Distribution:=option{1.0}
																	Padding:=margin{Left:=0.0, Top:=0.0, Right:=0.0, Bottom:=10.0}
																	Widget:=ConstructBlueStackBox()
										
										#Room layout indicator
										stack_box_slot:
											VerticalAlignment:=vertical_alignment.Center
											HorizontalAlignment:=horizontal_alignment.Center
											#Distribution:=option{1.0}
											Padding:=margin{Left:=0.0, Top:=0.0, Right:=5.0, Bottom:=15.0}
											Widget:=overlay:
												Slots:=array:
													
													#Makeshift Size box
													overlay_slot:
														VerticalAlignment:=vertical_alignment.Center
														HorizontalAlignment:=horizontal_alignment.Center
														#Distribution:=option{1.0}
														Padding:=margin{Left:=0.0, Top:=0.0, Right:=0.0, Bottom:=0.0}
														Widget:=texture_block{DefaultImage:=Textures.General.T_GenSquTrans, DefaultDesiredSize:=vector2{X:=350.0, Y:=35.0}, DefaultTint:=MakeColorFromHex("#00FF00FF")}

													#horizontal stack box for rooms layout
													overlay_slot:
														VerticalAlignment:=vertical_alignment.Center
														HorizontalAlignment:=horizontal_alignment.Center
														#Distribution:=option{1.0}
														Padding:=margin{Left:=0.0, Top:=0.0, Right:=0.0, Bottom:=0.0}
														Widget:=stack_box:
															Orientation:=orientation.Horizontal
															Slots:=array:

																#Preroom slot, R
																stack_box_slot:
																	VerticalAlignment:=vertical_alignment.Center
																	HorizontalAlignment:=horizontal_alignment.Center
																	Distribution:=option{1.0}
																	Padding:=margin{Left:=0.0, Top:=0.0, Right:=0.0, Bottom:=0.0}
																	Widget:=overlay:
																		Slots:=array:

																			overlay_slot:
																				VerticalAlignment:=vertical_alignment.Center
																				HorizontalAlignment:=horizontal_alignment.Center
																				#Distribution:=option{1.0}
																				Padding:=margin{Left:=0.0, Top:=0.0, Right:=0.0, Bottom:=0.0}
																				Widget:=texture_block{DefaultImage:=Textures.General.T_GenericSquare, DefaultDesiredSize:=vector2{X:=22.0, Y:=24.0}, DefaultTint:=MakeColorFromHex("#FFFFFFFF")}

																			overlay_slot:
																				VerticalAlignment:=vertical_alignment.Center
																				HorizontalAlignment:=horizontal_alignment.Center
																				#Distribution:=option{1.0}
																				Padding:=margin{Left:=0.0, Top:=0.0, Right:=0.0, Bottom:=0.0}
																				Widget:=text_block{DefaultText := StringToMessage("R"), DefaultTextColor:=MakeColorFromHex("#db242aFF"), DefaultShadowColor:=MakeColorFromHex("#000000FF"), DefaultShadowOffset:=option{vector2{X:=0.0, Y:=0.0}}}
																
																#Room slot, red
																stack_box_slot:
																	VerticalAlignment:=vertical_alignment.Center
																	HorizontalAlignment:=horizontal_alignment.Center
																	Distribution:=option{1.0}
																	Padding:=margin{Left:=0.0, Top:=0.0, Right:=0.0, Bottom:=0.0}
																	Widget:=overlay:
																		Slots:=array:

																			overlay_slot:
																				VerticalAlignment:=vertical_alignment.Center
																				HorizontalAlignment:=horizontal_alignment.Center
																				#Distribution:=option{1.0}
																				Padding:=margin{Left:=0.0, Top:=0.0, Right:=0.0, Bottom:=0.0}
																				Widget:=texture_block{DefaultImage:=Textures.General.T_GenericSquare, DefaultDesiredSize:=vector2{X:=22.0, Y:=24.0}, DefaultTint:=MakeColorFromHex("#FFFFFFFF")}

																			overlay_slot:
																				VerticalAlignment:=vertical_alignment.Center
																				HorizontalAlignment:=horizontal_alignment.Center
																				#Distribution:=option{1.0}
																				Padding:=margin{Left:=0.0, Top:=0.0, Right:=0.0, Bottom:=0.0}
																				Widget:= RoomRedSide1

																#Room slot, Grey
																stack_box_slot:
																	VerticalAlignment:=vertical_alignment.Center
																	HorizontalAlignment:=horizontal_alignment.Center
																	Distribution:=option{1.0}
																	Padding:=margin{Left:=0.0, Top:=0.0, Right:=0.0, Bottom:=0.0}
																	Widget:=overlay:
																		Slots:=array:

																			overlay_slot:
																				VerticalAlignment:=vertical_alignment.Center
																				HorizontalAlignment:=horizontal_alignment.Center
																				#Distribution:=option{1.0}
																				Padding:=margin{Left:=0.0, Top:=0.0, Right:=0.0, Bottom:=0.0}
																				Widget:=texture_block{DefaultImage:=Textures.General.T_GenericSquare, DefaultDesiredSize:=vector2{X:=22.0, Y:=24.0}, DefaultTint:=MakeColorFromHex("#FFFFFFFF")}

																			overlay_slot:
																				VerticalAlignment:=vertical_alignment.Center
																				HorizontalAlignment:=horizontal_alignment.Center
																				#Distribution:=option{1.0}
																				Padding:=margin{Left:=0.0, Top:=0.0, Right:=0.0, Bottom:=0.0}
																				Widget:= RoomRedSide2 

																#Room slot, Grey
																stack_box_slot:
																	VerticalAlignment:=vertical_alignment.Center
																	HorizontalAlignment:=horizontal_alignment.Center
																	Distribution:=option{1.0}
																	Padding:=margin{Left:=0.0, Top:=0.0, Right:=0.0, Bottom:=0.0}
																	Widget:=overlay:
																		Slots:=array:

																			overlay_slot:
																				VerticalAlignment:=vertical_alignment.Center
																				HorizontalAlignment:=horizontal_alignment.Center
																				#Distribution:=option{1.0}
																				Padding:=margin{Left:=0.0, Top:=0.0, Right:=0.0, Bottom:=0.0}
																				Widget:=texture_block{DefaultImage:=Textures.General.T_GenericSquare, DefaultDesiredSize:=vector2{X:=22.0, Y:=24.0}, DefaultTint:=MakeColorFromHex("#FFFFFFFF")}

																			overlay_slot:
																				VerticalAlignment:=vertical_alignment.Center
																				HorizontalAlignment:=horizontal_alignment.Center
																				#Distribution:=option{1.0}
																				Padding:=margin{Left:=0.0, Top:=0.0, Right:=0.0, Bottom:=0.0}
																				Widget:= RoomRedSide3

																#Room slot, Grey
																stack_box_slot:
																	VerticalAlignment:=vertical_alignment.Center
																	HorizontalAlignment:=horizontal_alignment.Center
																	Distribution:=option{1.0}
																	Padding:=margin{Left:=0.0, Top:=0.0, Right:=0.0, Bottom:=0.0}
																	Widget:=overlay:
																		Slots:=array:

																			overlay_slot:
																				VerticalAlignment:=vertical_alignment.Center
																				HorizontalAlignment:=horizontal_alignment.Center
																				#Distribution:=option{1.0}
																				Padding:=margin{Left:=0.0, Top:=0.0, Right:=0.0, Bottom:=0.0}
																				Widget:=texture_block{DefaultImage:=Textures.General.T_GenericSquare, DefaultDesiredSize:=vector2{X:=22.0, Y:=24.0}, DefaultTint:=MakeColorFromHex("#FFFFFFFF")}

																			overlay_slot:
																				VerticalAlignment:=vertical_alignment.Center
																				HorizontalAlignment:=horizontal_alignment.Center
																				#Distribution:=option{1.0}
																				Padding:=margin{Left:=0.0, Top:=0.0, Right:=0.0, Bottom:=0.0}
																				Widget:= RoomRedSide4

																#Halfway room slot, Grey
																stack_box_slot:
																	VerticalAlignment:=vertical_alignment.Center
																	HorizontalAlignment:=horizontal_alignment.Fill
																	#Distribution:=option{1.0}
																	Padding:=margin{Left:=0.0, Top:=0.0, Right:=0.0, Bottom:=0.0}
																	Widget:=overlay:
																		Slots:=array:

																			overlay_slot:
																				VerticalAlignment:=vertical_alignment.Center
																				HorizontalAlignment:=horizontal_alignment.Center
																				#Distribution:=option{1.0}
																				Padding:=margin{Left:=0.0, Top:=0.0, Right:=0.0, Bottom:=0.0}
																				Widget:=texture_block{DefaultImage:=Textures.General.T_GenericSquare, DefaultDesiredSize:=vector2{X:=50.0, Y:=30.0}, DefaultTint:=MakeColorFromHex("#FFFFFFFF")}

																			overlay_slot:
																				VerticalAlignment:=vertical_alignment.Center
																				HorizontalAlignment:=horizontal_alignment.Center
																				#Distribution:=option{1.0}
																				Padding:=margin{Left:=0.0, Top:=0.0, Right:=0.0, Bottom:=0.0}
																				Widget:= RoomArena

																#Room slot, Grey
																stack_box_slot:
																	VerticalAlignment:=vertical_alignment.Center
																	HorizontalAlignment:=horizontal_alignment.Center
																	Distribution:=option{1.0}
																	Padding:=margin{Left:=0.0, Top:=0.0, Right:=0.0, Bottom:=0.0}
																	Widget:=overlay:
																		Slots:=array:

																			overlay_slot:
																				VerticalAlignment:=vertical_alignment.Center
																				HorizontalAlignment:=horizontal_alignment.Center
																				#Distribution:=option{1.0}
																				Padding:=margin{Left:=0.0, Top:=0.0, Right:=0.0, Bottom:=0.0}
																				Widget:=texture_block{DefaultImage:=Textures.General.T_GenericSquare, DefaultDesiredSize:=vector2{X:=22.0, Y:=24.0}, DefaultTint:=MakeColorFromHex("#FFFFFFFF")}

																			overlay_slot:
																				VerticalAlignment:=vertical_alignment.Center
																				HorizontalAlignment:=horizontal_alignment.Center
																				#Distribution:=option{1.0}
																				Padding:=margin{Left:=0.0, Top:=0.0, Right:=0.0, Bottom:=0.0}
																				Widget:= RoomBlueSide1

																#Room slot, Blue
																stack_box_slot:
																	VerticalAlignment:=vertical_alignment.Center
																	HorizontalAlignment:=horizontal_alignment.Center
																	Distribution:=option{1.0}
																	Padding:=margin{Left:=0.0, Top:=0.0, Right:=0.0, Bottom:=0.0}
																	Widget:=overlay:
																		Slots:=array:

																			overlay_slot:
																				VerticalAlignment:=vertical_alignment.Center
																				HorizontalAlignment:=horizontal_alignment.Center
																				#Distribution:=option{1.0}
																				Padding:=margin{Left:=0.0, Top:=0.0, Right:=0.0, Bottom:=0.0}
																				Widget:=texture_block{DefaultImage:=Textures.General.T_GenericSquare, DefaultDesiredSize:=vector2{X:=22.0, Y:=24.0}, DefaultTint:=MakeColorFromHex("#FFFFFFFF")}

																			overlay_slot:
																				VerticalAlignment:=vertical_alignment.Center
																				HorizontalAlignment:=horizontal_alignment.Center
																				#Distribution:=option{1.0}
																				Padding:=margin{Left:=0.0, Top:=0.0, Right:=0.0, Bottom:=0.0}
																				Widget:= RoomBlueSide2

																#Room slot, Blue
																stack_box_slot:
																	VerticalAlignment:=vertical_alignment.Center
																	HorizontalAlignment:=horizontal_alignment.Center
																	Distribution:=option{1.0}
																	Padding:=margin{Left:=0.0, Top:=0.0, Right:=0.0, Bottom:=0.0}
																	Widget:=overlay:
																		Slots:=array:

																			overlay_slot:
																				VerticalAlignment:=vertical_alignment.Center
																				HorizontalAlignment:=horizontal_alignment.Center
																				#Distribution:=option{1.0}
																				Padding:=margin{Left:=0.0, Top:=0.0, Right:=0.0, Bottom:=0.0}
																				Widget:=texture_block{DefaultImage:=Textures.General.T_GenericSquare, DefaultDesiredSize:=vector2{X:=22.0, Y:=24.0}, DefaultTint:=MakeColorFromHex("#FFFFFFFF")}

																			overlay_slot:
																				VerticalAlignment:=vertical_alignment.Center
																				HorizontalAlignment:=horizontal_alignment.Center
																				#Distribution:=option{1.0}
																				Padding:=margin{Left:=0.0, Top:=0.0, Right:=0.0, Bottom:=0.0}
																				Widget:= RoomBlueSide3

																#Room slot, Blue
																stack_box_slot:
																	VerticalAlignment:=vertical_alignment.Center
																	HorizontalAlignment:=horizontal_alignment.Center
																	Distribution:=option{1.0}
																	Padding:=margin{Left:=0.0, Top:=0.0, Right:=0.0, Bottom:=0.0}
																	Widget:=overlay:
																		Slots:=array:

																			overlay_slot:
																				VerticalAlignment:=vertical_alignment.Center
																				HorizontalAlignment:=horizontal_alignment.Center
																				#Distribution:=option{1.0}
																				Padding:=margin{Left:=0.0, Top:=0.0, Right:=0.0, Bottom:=0.0}
																				Widget:=texture_block{DefaultImage:=Textures.General.T_GenericSquare, DefaultDesiredSize:=vector2{X:=22.0, Y:=24.0}, DefaultTint:=MakeColorFromHex("#FFFFFFFF")}

																			overlay_slot:
																				VerticalAlignment:=vertical_alignment.Center
																				HorizontalAlignment:=horizontal_alignment.Center
																				#Distribution:=option{1.0}
																				Padding:=margin{Left:=0.0, Top:=0.0, Right:=0.0, Bottom:=0.0}
																				Widget:= RoomBlueSide4

																#Preroom slot, B
																stack_box_slot:
																	VerticalAlignment:=vertical_alignment.Center
																	HorizontalAlignment:=horizontal_alignment.Center
																	Distribution:=option{1.0}
																	Padding:=margin{Left:=0.0, Top:=0.0, Right:=0.0, Bottom:=0.0}
																	Widget:=overlay:
																		Slots:=array:

																			overlay_slot:
																				VerticalAlignment:=vertical_alignment.Center
																				HorizontalAlignment:=horizontal_alignment.Center
																				#Distribution:=option{1.0}
																				Padding:=margin{Left:=0.0, Top:=0.0, Right:=0.0, Bottom:=0.0}
																				Widget:=texture_block{DefaultImage:=Textures.General.T_GenericSquare, DefaultDesiredSize:=vector2{X:=22.0, Y:=24.0}, DefaultTint:=MakeColorFromHex("#FFFFFFFF")}

																			overlay_slot:
																				VerticalAlignment:=vertical_alignment.Center
																				HorizontalAlignment:=horizontal_alignment.Center
																				#Distribution:=option{1.0}
																				Padding:=margin{Left:=0.0, Top:=0.0, Right:=0.0, Bottom:=0.0}
																				Widget:=text_block{DefaultText := StringToMessage("B"), DefaultTextColor:=MakeColorFromHex("#417af4FF"), DefaultShadowColor:=MakeColorFromHex("#000000FF"), DefaultShadowOffset:=option{vector2{X:=0.0, Y:=0.0}}}
		return HudOverlay

	ConstructBlueStackBox():stack_box=
		set BlueBedStackBox = stack_box:
			Orientation:=orientation.Horizontal
			Slots:=array:
				#Big room module - Blue
				stack_box_slot:
					VerticalAlignment:=vertical_alignment.Fill
					HorizontalAlignment:=horizontal_alignment.Fill
					#Distribution:=option{1.0}
					Padding:=margin{Left:=0.0, Top:=0.0, Right:=7.0, Bottom:=0.0}
					Widget:=overlay:
						Slots:=array:

							#Big Slot
							overlay_slot:
								VerticalAlignment:=vertical_alignment.Center
								HorizontalAlignment:=horizontal_alignment.Left
								#Distribution:=option{1.0}
								Padding:=margin{Left:=0.0, Top:=8.0, Right:=0.0, Bottom:=0.0}
								Widget:=overlay:
									Slots:=array:
										
										#outher diamond texture
										overlay_slot:
											VerticalAlignment:=vertical_alignment.Center
											HorizontalAlignment:=horizontal_alignment.Center
											#Distribution:=option{1.0}
											Padding:=margin{Left:=0.0, Top:=0.0, Right:=0.0, Bottom:=0.0}
											Widget:=texture_block{DefaultImage:=Textures.General.T_GenTooltip, DefaultDesiredSize:=vector2{X:=62.0, Y:=62.0}, DefaultTint:=MakeColorFromHex("#FFFFFFFF")}

										#Inner diamond texture
										overlay_slot:
											VerticalAlignment:=vertical_alignment.Center
											HorizontalAlignment:=horizontal_alignment.Center
											#Distribution:=option{1.0}
											Padding:=margin{Left:=0.0, Top:=1.0, Right:=0.0, Bottom:=2.0}
											Widget:=texture_block{DefaultImage:=Textures.General.T_GenTooltip, DefaultDesiredSize:=vector2{X:=54.0, Y:=52.0}, DefaultTint:=MakeColorFromHex("#417af4FF")}

										#Bed icon texture
										overlay_slot:
											VerticalAlignment:=vertical_alignment.Top
											HorizontalAlignment:=horizontal_alignment.Center
											#Distribution:=option{1.0}
											Padding:=margin{Left:=0.0, Top:=10.0, Right:=3.0, Bottom:=0.0}
											Widget:=texture_block{DefaultImage:=Textures.General.T_UI_Icon_bed, DefaultDesiredSize:=vector2{X:=30.0, Y:=30.0}, DefaultTint:=MakeColorFromHex("#FFFFFFFF")}
				
				#Small room arrow - red highlight
				stack_box_slot:
					VerticalAlignment:=vertical_alignment.Fill
					HorizontalAlignment:=horizontal_alignment.Fill
					#Distribution:=option{1.0}
					Padding:=margin{Left:=0.0, Top:=0.0, Right:=7.0, Bottom:=0.0}
					Widget:=overlay:
						Slots:=array:

							#Background diamond texture
							overlay_slot:
								VerticalAlignment:=vertical_alignment.Center
								HorizontalAlignment:=horizontal_alignment.Center
								#Distribution:=option{1.0}
								Padding:=margin{Left:=0.0, Top:=0.0, Right:=0.0, Bottom:=0.0}
								Widget:=texture_block{DefaultImage:=Textures.General.T_GenArrowLeft, DefaultDesiredSize:=vector2{X:=56.0, Y:=36.0}, DefaultTint:=MakeColorFromHex("#FFFFFFFF")}

							#Inner diamond texture
							overlay_slot:
								VerticalAlignment:=vertical_alignment.Center
								HorizontalAlignment:=horizontal_alignment.Center
								#Distribution:=option{1.0}
								Padding:=margin{Left:=1.0, Top:=0.0, Right:=2.0, Bottom:=0.0}
								Widget:= BlueBed4

							#Bed icon texture
							overlay_slot:
								VerticalAlignment:=vertical_alignment.Center
								HorizontalAlignment:=horizontal_alignment.Center
								#Distribution:=option{1.0}
								Padding:=margin{Left:=0.0, Top:=0.0, Right:=3.0, Bottom:=0.0}
								Widget:=texture_block{DefaultImage:=Textures.General.T_UI_Icon_bed, DefaultDesiredSize:=vector2{X:=24.0, Y:=24.0}, DefaultTint:=MakeColorFromHex("#FFFFFFFF")}
																								
				#Small room arrow - Red
				stack_box_slot:
					VerticalAlignment:=vertical_alignment.Fill
					HorizontalAlignment:=horizontal_alignment.Fill
					#Distribution:=option{1.0}
					Padding:=margin{Left:=1.0, Top:=0.0, Right:=7.0, Bottom:=0.0}
					Widget:=overlay:
						Slots:=array:

							#Background diamond texture
							overlay_slot:
								VerticalAlignment:=vertical_alignment.Center
								HorizontalAlignment:=horizontal_alignment.Center
								#Distribution:=option{1.0}
								Padding:=margin{Left:=0.0, Top:=0.0, Right:=0.0, Bottom:=0.0}
								Widget:=texture_block{DefaultImage:=Textures.General.T_GenArrowLeft, DefaultDesiredSize:=vector2{X:=56.0, Y:=36.0}, DefaultTint:=MakeColorFromHex("#FFFFFFFF")}

							#Inner diamond texture
							overlay_slot:
								VerticalAlignment:=vertical_alignment.Center
								HorizontalAlignment:=horizontal_alignment.Center
								#Distribution:=option{1.0}
								Padding:=margin{Left:=1.0, Top:=0.0, Right:=2.0, Bottom:=0.0}
								Widget:=BlueBed3

							#Bed icon texture
							overlay_slot:
								VerticalAlignment:=vertical_alignment.Center
								HorizontalAlignment:=horizontal_alignment.Center
								#Distribution:=option{1.0}
								Padding:=margin{Left:=0.0, Top:=0.0, Right:=3.0, Bottom:=0.0}
								Widget:=texture_block{DefaultImage:=Textures.General.T_UI_Icon_bed, DefaultDesiredSize:=vector2{X:=24.0, Y:=24.0}, DefaultTint:=MakeColorFromHex("#FFFFFFFF")}

				#Small room arrow - Red
				#stack_box_slot:
					#VerticalAlignment:=vertical_alignment.Fill
					#HorizontalAlignment:=horizontal_alignment.Fill
					#Distribution:=option{1.0}
					#Padding:=margin{Left:=1.0, Top:=0.0, Right:=7.0, Bottom:=0.0}
					#Widget:=overlay:
						#Slots:=array:

							#Background diamond texture
							#overlay_slot:
								#VerticalAlignment:=vertical_alignment.Center
								#HorizontalAlignment:=horizontal_alignment.Center
								#Distribution:=option{1.0}
								#Padding:=margin{Left:=0.0, Top:=0.0, Right:=0.0, Bottom:=0.0}
								#Widget:=texture_block{DefaultImage:=Textures.General.T_GenArrowLeft, DefaultDesiredSize:=vector2{X:=56.0, Y:=36.0}, DefaultTint:=MakeColorFromHex("#FFFFFFFF")}

							#Inner diamond texture
							#overlay_slot:
								#VerticalAlignment:=vertical_alignment.Center
								#HorizontalAlignment:=horizontal_alignment.Center
								#Distribution:=option{1.0}
								#Padding:=margin{Left:=1.0, Top:=0.0, Right:=2.0, Bottom:=0.0}
								#Widget:=BlueBed2

							#Bed icon texture
							#overlay_slot:
								#VerticalAlignment:=vertical_alignment.Center
								#HorizontalAlignment:=horizontal_alignment.Center
								#Distribution:=option{1.0}
								#Padding:=margin{Left:=0.0, Top:=0.0, Right:=3.0, Bottom:=0.0}
								#Widget:=texture_block{DefaultImage:=Textures.General.T_UI_Icon_bed, DefaultDesiredSize:=vector2{X:=24.0, Y:=24.0}, DefaultTint:=MakeColorFromHex("#FFFFFFFF")}

				#Small room arrow - Red
				#stack_box_slot:
					#VerticalAlignment:=vertical_alignment.Fill
					#HorizontalAlignment:=horizontal_alignment.Fill
					#Distribution:=option{1.0}
					#Padding:=margin{Left:=1.0, Top:=0.0, Right:=7.0, Bottom:=0.0}
					#Widget:=overlay:
						#Slots:=array:

							#Background diamond texture
							#overlay_slot:
								#VerticalAlignment:=vertical_alignment.Center
								#HorizontalAlignment:=horizontal_alignment.Center
								#Distribution:=option{1.0}
								#Padding:=margin{Left:=0.0, Top:=0.0, Right:=0.0, Bottom:=0.0}
								#Widget:=texture_block{DefaultImage:=Textures.General.T_GenArrowLeft, DefaultDesiredSize:=vector2{X:=56.0, Y:=36.0}, DefaultTint:=MakeColorFromHex("#FFFFFFFF")}

							#Inner diamond texture
							#overlay_slot:
								#VerticalAlignment:=vertical_alignment.Center
								#HorizontalAlignment:=horizontal_alignment.Center
								#Distribution:=option{1.0}
								#Padding:=margin{Left:=1.0, Top:=0.0, Right:=2.0, Bottom:=0.0}
								#Widget:=BlueBed1

							#Bed icon texture
							#overlay_slot:
								#VerticalAlignment:=vertical_alignment.Center
								#HorizontalAlignment:=horizontal_alignment.Center
								#Distribution:=option{1.0}
								#Padding:=margin{Left:=0.0, Top:=0.0, Right:=3.0, Bottom:=0.0}
								#Widget:=texture_block{DefaultImage:=Textures.General.T_UI_Icon_bed, DefaultDesiredSize:=vector2{X:=24.0, Y:=24.0}, DefaultTint:=MakeColorFromHex("#FFFFFFFF")}
		return BlueBedStackBox


	ConstructRedStackBox():stack_box=
		set RedBedStackBox = stack_box:
				Orientation:=orientation.Horizontal
				Slots:=array:

					#Small room arrow - red highlight
					#stack_box_slot:
						#VerticalAlignment:=vertical_alignment.Fill
						#HorizontalAlignment:=horizontal_alignment.Fill
						#Distribution:=option{1.0}
						#Padding:=margin{Left:=0.0, Top:=0.0, Right:=7.0, Bottom:=0.0}
						#Widget:=overlay:
							#Slots:=array:

								#Background diamond texture
								#overlay_slot:
									#VerticalAlignment:=vertical_alignment.Center
									#HorizontalAlignment:=horizontal_alignment.Center
									#Distribution:=option{1.0}
									#Padding:=margin{Left:=0.0, Top:=0.0, Right:=0.0, Bottom:=0.0}
									#Widget:=texture_block{DefaultImage:=Textures.General.T_GenArrow, DefaultDesiredSize:=vector2{X:=56.0, Y:=36.0}, DefaultTint:=MakeColorFromHex("#FFFFFFFF")}

								#Inner diamond texture
								#overlay_slot:
									#VerticalAlignment:=vertical_alignment.Center
									#HorizontalAlignment:=horizontal_alignment.Center
									#Distribution:=option{1.0}
									#Padding:=margin{Left:=1.0, Top:=0.0, Right:=0.0, Bottom:=0.0}
									#Widget:= RedBed1

								#Bed icon texture
								#overlay_slot:
									#VerticalAlignment:=vertical_alignment.Center
									#HorizontalAlignment:=horizontal_alignment.Center
									#Distribution:=option{1.0}
									#Padding:=margin{Left:=3.0, Top:=0.0, Right:=0.0, Bottom:=0.0}
									#Widget:=texture_block{DefaultImage:=Textures.General.T_UI_Icon_bed_right, DefaultDesiredSize:=vector2{X:=24.0, Y:=24.0}, DefaultTint:=MakeColorFromHex("#FFFFFFFF")}
					
					#Small room arrow - Red
					#stack_box_slot:
						#VerticalAlignment:=vertical_alignment.Fill
						#HorizontalAlignment:=horizontal_alignment.Fill
						#Distribution:=option{1.0}
						#Padding:=margin{Left:=1.0, Top:=0.0, Right:=7.0, Bottom:=0.0}
						#Widget:=overlay:
							#Slots:=array:

								#Background diamond texture
								#overlay_slot:
									#VerticalAlignment:=vertical_alignment.Center
									#HorizontalAlignment:=horizontal_alignment.Center
									#Distribution:=option{1.0}
									#Padding:=margin{Left:=0.0, Top:=0.0, Right:=0.0, Bottom:=0.0}
									#Widget:=texture_block{DefaultImage:=Textures.General.T_GenArrow, DefaultDesiredSize:=vector2{X:=56.0, Y:=36.0}, DefaultTint:=MakeColorFromHex("#FFFFFFFF")}

								#Inner diamond texture
								#overlay_slot:
									#VerticalAlignment:=vertical_alignment.Center
									#HorizontalAlignment:=horizontal_alignment.Center
									#Distribution:=option{1.0}
									#Padding:=margin{Left:=1.0, Top:=0.0, Right:=0.0, Bottom:=0.0}
									#Widget:= RedBed2

								#Bed icon texture
								#overlay_slot:
									#VerticalAlignment:=vertical_alignment.Center
									#HorizontalAlignment:=horizontal_alignment.Center
									#Distribution:=option{1.0}
									#Padding:=margin{Left:=3.0, Top:=0.0, Right:=0.0, Bottom:=0.0}
									#Widget:=texture_block{DefaultImage:=Textures.General.T_UI_Icon_bed_right, DefaultDesiredSize:=vector2{X:=24.0, Y:=24.0}, DefaultTint:=MakeColorFromHex("#FFFFFFFF")}

					#Small room arrow - Red
					stack_box_slot:
						VerticalAlignment:=vertical_alignment.Fill
						HorizontalAlignment:=horizontal_alignment.Fill
						#Distribution:=option{1.0}
						Padding:=margin{Left:=1.0, Top:=0.0, Right:=7.0, Bottom:=0.0}
						Widget:=overlay:
							Slots:=array:

								#Background diamond texture
								overlay_slot:
									VerticalAlignment:=vertical_alignment.Center
									HorizontalAlignment:=horizontal_alignment.Center
									#Distribution:=option{1.0}
									Padding:=margin{Left:=0.0, Top:=0.0, Right:=0.0, Bottom:=0.0}
									Widget:=texture_block{DefaultImage:=Textures.General.T_GenArrow, DefaultDesiredSize:=vector2{X:=56.0, Y:=36.0}, DefaultTint:=MakeColorFromHex("#FFFFFFFF")}

								#Inner diamond texture
								overlay_slot:
									VerticalAlignment:=vertical_alignment.Center
									HorizontalAlignment:=horizontal_alignment.Center
									#Distribution:=option{1.0}
									Padding:=margin{Left:=1.0, Top:=0.0, Right:=0.0, Bottom:=0.0}
									Widget:=RedBed3

								#Bed icon texture
								overlay_slot:
									VerticalAlignment:=vertical_alignment.Center
									HorizontalAlignment:=horizontal_alignment.Center
									#Distribution:=option{1.0}
									Padding:=margin{Left:=3.0, Top:=0.0, Right:=0.0, Bottom:=0.0}
									Widget:=texture_block{DefaultImage:=Textures.General.T_UI_Icon_bed_right, DefaultDesiredSize:=vector2{X:=24.0, Y:=24.0}, DefaultTint:=MakeColorFromHex("#FFFFFFFF")}

					#Small room arrow - Red
					stack_box_slot:
						VerticalAlignment:=vertical_alignment.Fill
						HorizontalAlignment:=horizontal_alignment.Fill
						#Distribution:=option{1.0}
						Padding:=margin{Left:=1.0, Top:=0.0, Right:=7.0, Bottom:=0.0}
						Widget:=overlay:
							Slots:=array:

								#Background diamond texture
								overlay_slot:
									VerticalAlignment:=vertical_alignment.Center
									HorizontalAlignment:=horizontal_alignment.Center
									#Distribution:=option{1.0}
									Padding:=margin{Left:=0.0, Top:=0.0, Right:=0.0, Bottom:=0.0}
									Widget:=texture_block{DefaultImage:=Textures.General.T_GenArrow, DefaultDesiredSize:=vector2{X:=56.0, Y:=36.0}, DefaultTint:=MakeColorFromHex("#FFFFFFFF")}

								#Inner diamond texture
								overlay_slot:
									VerticalAlignment:=vertical_alignment.Center
									HorizontalAlignment:=horizontal_alignment.Center
									#Distribution:=option{1.0}
									Padding:=margin{Left:=1.0, Top:=0.0, Right:=0.0, Bottom:=0.0}
									Widget:=RedBed4

								#Bed icon texture
								overlay_slot:
									VerticalAlignment:=vertical_alignment.Center
									HorizontalAlignment:=horizontal_alignment.Center
									#Distribution:=option{1.0}
									Padding:=margin{Left:=3.0, Top:=0.0, Right:=0.0, Bottom:=0.0}
									Widget:=texture_block{DefaultImage:=Textures.General.T_UI_Icon_bed_right, DefaultDesiredSize:=vector2{X:=24.0, Y:=24.0}, DefaultTint:=MakeColorFromHex("#FFFFFFFF")}

					#Big room module - Red
					stack_box_slot:
						VerticalAlignment:=vertical_alignment.Fill
						HorizontalAlignment:=horizontal_alignment.Fill
						#Distribution:=option{1.0}
						Padding:=margin{Left:=0.0, Top:=0.0, Right:=7.0, Bottom:=0.0}
						Widget:=overlay:
							Slots:=array:

								#Big Slot
								overlay_slot:
									VerticalAlignment:=vertical_alignment.Center
									HorizontalAlignment:=horizontal_alignment.Left
									#Distribution:=option{1.0}
									Padding:=margin{Left:=0.0, Top:=8.0, Right:=0.0, Bottom:=0.0}
									Widget:=overlay:
										Slots:=array:
											
											#outher diamond texture
											overlay_slot:
												VerticalAlignment:=vertical_alignment.Center
												HorizontalAlignment:=horizontal_alignment.Center
												#Distribution:=option{1.0}
												Padding:=margin{Left:=0.0, Top:=0.0, Right:=0.0, Bottom:=0.0}
												Widget:=texture_block{DefaultImage:=Textures.General.T_GenTooltip, DefaultDesiredSize:=vector2{X:=62.0, Y:=62.0}, DefaultTint:=MakeColorFromHex("#FFFFFFFF")}

											#Inner diamond texture
											overlay_slot:
												VerticalAlignment:=vertical_alignment.Center
												HorizontalAlignment:=horizontal_alignment.Center
												#Distribution:=option{1.0}
												Padding:=margin{Left:=0.0, Top:=1.0, Right:=0.0, Bottom:=2.0}
												Widget:=texture_block{DefaultImage:=Textures.General.T_GenTooltip, DefaultDesiredSize:=vector2{X:=54.0, Y:=52.0}, DefaultTint:=MakeColorFromHex("#db242aFF")}

											#Bed icon texture
											overlay_slot:
												VerticalAlignment:=vertical_alignment.Top
												HorizontalAlignment:=horizontal_alignment.Center
												#Distribution:=option{1.0}
												Padding:=margin{Left:=0.0, Top:=10.0, Right:=0.0, Bottom:=0.0}
												Widget:=texture_block{DefaultImage:=Textures.General.T_UI_Icon_bed_right, DefaultDesiredSize:=vector2{X:=30.0, Y:=30.0}, DefaultTint:=MakeColorFromHex("#FFFFFFFF")}

		return RedBedStackBox

	# For the rooms list, it will always go from room one in the blue side until the last room in the red side
	#-----------------------------------------------------------------------------------------------------------------------
	UpdateRooms(Rooms : []team_id): void =
	{
		Print("UPDATING ROOMS IN ui_hud_match_maker!")
		# TODO: We need to escale this to handle separate rooms
		if:
			Team := Rooms[0]
			Color := GetTeamColor(Team)
		then:
			RoomBlueSide4.SetTint(GetTeamColor(Team))

		if:
			Team := Rooms[1]
			Color := GetTeamColor(Team)
		then:
			RoomBlueSide3.SetTint(GetTeamColor(Team))

		if:
			Team := Rooms[2]
			Color := GetTeamColor(Team)
		then:
			RoomBlueSide2.SetTint(GetTeamColor(Team))

		if:
			Team := Rooms[3]
			Color := GetTeamColor(Team)
		then:
			RoomBlueSide1.SetTint(GetTeamColor(Team))

		if:
			Team := Rooms[4]
			Color := GetTeamColor(Team)
		then:
			RoomArena.SetTint(GetTeamColor(Team))

		if:
			Team := Rooms[5]
			Color := GetTeamColor(Team)
		then:
			RoomRedSide4.SetTint(GetTeamColor(Team))

		if:
			Team := Rooms[6]
			Color := GetTeamColor(Team)
		then:
			RoomRedSide3.SetTint(GetTeamColor(Team))

		if:
			Team := Rooms[7]
			Color := GetTeamColor(Team)
		then:
			RoomRedSide2.SetTint(GetTeamColor(Team))

		if:
			Team := Rooms[8]
			Color := GetTeamColor(Team)
		then:
			RoomRedSide1.SetTint(GetTeamColor(Team))


		#RoomBlueSide1.SetTint(GetTeamColor(team_id.Red))

	}
	
	# Update Score Values
	#-----------------------------------------------------------------------------------------------------------------------
	UpdateDestroyedBeds(DestroyerTeam:team_id, AmountDestroyed:int):void=
		var BedToRemoveIndex:int = 0
		var StackBoxToUpdate:stack_box = stack_box{Orientation:=orientation.Horizontal}
		case(DestroyerTeam):
			team_id.Blue =>
				set StackBoxToUpdate = RedBedStackBox
				set BedToRemoveIndex = AmountDestroyed - 1
			team_id.Red =>
				set StackBoxToUpdate = BlueBedStackBox
				set BedToRemoveIndex = AmountOfBedsToShow - AmountDestroyed
			_ =>
				Print("Invalid Team")

		SlotsToUpdate := StackBoxToUpdate.Slots
		if:
			WidgetToUpdate := SlotsToUpdate[BedToRemoveIndex].Widget
			OverlayToUpdate := overlay[WidgetToUpdate]
			BackbroundDiamondSlot := OverlayToUpdate.Slots[0].Widget
			InnerDiamondSlot := OverlayToUpdate.Slots[1].Widget
			BedIconSlot := OverlayToUpdate.Slots[2].Widget
		then:
			BackbroundDiamondSlot.SetVisibility(widget_visibility.Hidden)
			InnerDiamondSlot.SetVisibility(widget_visibility.Hidden)
			BedIconSlot.SetVisibility(widget_visibility.Hidden)

	#-----------------------------------------------------------------------------------------------------------------------
	GetTeamColor(RoomTeam : team_id)<transacts>: color =
	{
		#db242aFF #c5c6c8FF #417af4FF

		Print("room team {RoomTeam}")
		case (RoomTeam):
			team_id.Blue =>
				return MakeColorFromHex("#417af4FF")
			team_id.Red =>
				return MakeColorFromHex("#db242aFF")
			_=>
				return MakeColorFromHex("#c5c6c8FF")
	}

	UpdateRoomName(NewRoomName : string): void =
		RoomName.SetText(StringToMessage(NewRoomName))


<#

ui_hud_match_marker := class(creative_device):#class(ui_common)

	OnBegin<override>()<suspends>:void=
		CreateCanvas()


	CreateCanvas():void=
		Players:=GetPlayspace().GetPlayers()
		if(Player:=Players[0]):
			
			#Main structure for the match marker
			MatchMarker:=overlay:
				Slots:=array:
					#Makeshift size box
					overlay_slot:
						VerticalAlignment:=vertical_alignment.Center
						HorizontalAlignment:=horizontal_alignment.Fill
						#Distribution:=option{0.5}
						Padding:=margin{Left:=0.0, Top:=0.0, Right:=0.0, Bottom:=0.0}
						Widget:=texture_block{DefaultImage:=Textures.General.T_GenSquTrans, DefaultDesiredSize:=vector2{X:=552.0, Y:=115.0}, DefaultTint:=MakeColorFromHex("#FF000066")}

					#Main layout
					overlay_slot:
						VerticalAlignment:=vertical_alignment.Top
						HorizontalAlignment:=horizontal_alignment.Center
						#Distribution:=option{0.5}
						Padding:=margin{Left:=0.0, Top:=0.0, Right:=0.0, Bottom:=0.0}
						Widget:=stack_box:
							Orientation:=orientation.Vertical
							Slots:=array:

								#Actual progress module
								stack_box_slot:
									VerticalAlignment:=vertical_alignment.Center
									HorizontalAlignment:=horizontal_alignment.Center
									Distribution:=option{1.0}
									Padding:=margin{Left:=0.0, Top:=0.0, Right:=0.0, Bottom:=0.0}
									Widget:=stack_box:
										Orientation:=orientation.Vertical
										Slots:=array:

											#Match progress module
											stack_box_slot:
												VerticalAlignment:=vertical_alignment.Center
												HorizontalAlignment:=horizontal_alignment.Center
												#Distribution:=option{1.0}
												Padding:=margin{Left:=0.0, Top:=0.0, Right:=0.0, Bottom:=0.0}
												Widget:=stack_box:
													Orientation:=orientation.Horizontal
													Slots:=array:
														
														#Red team beds
														stack_box_slot:
															VerticalAlignment:=vertical_alignment.Center
															HorizontalAlignment:=horizontal_alignment.Left
															#Distribution:=option{1.0}
															Padding:=margin{Left:=0.0, Top:=0.0, Right:=0.0, Bottom:=0.0}
															Widget:=overlay:
																Slots:=array:
																	
																	#Makeshift size box for first team beds
																	overlay_slot:
																		VerticalAlignment:=vertical_alignment.Center
																		HorizontalAlignment:=horizontal_alignment.Center
																		#Distribution:=option{1.0}
																		Padding:=margin{Left:=0.0, Top:=0.0, Right:=0.0, Bottom:=0.0}
																		Widget:=texture_block{DefaultImage:=Textures.General.T_GenSquTrans, DefaultDesiredSize:=vector2{X:=315.0, Y:=65.0}, DefaultTint:=MakeColorFromHex("#0000FFFF")}

																	#Actual content for first team beds
																	overlay_slot:
																		VerticalAlignment:=vertical_alignment.Center
																		HorizontalAlignment:=horizontal_alignment.Right
																		#Distribution:=option{1.0}
																		Padding:=margin{Left:=0.0, Top:=0.0, Right:=0.0, Bottom:=10.0}
																		Widget:=stack_box:
																			Orientation:=orientation.Horizontal
																			Slots:=array:

																				#Small room arrow - red highlight
																				stack_box_slot:
																					VerticalAlignment:=vertical_alignment.Fill
																					HorizontalAlignment:=horizontal_alignment.Fill
																					#Distribution:=option{1.0}
																					Padding:=margin{Left:=0.0, Top:=0.0, Right:=7.0, Bottom:=0.0}
																					Widget:=overlay:
																						Slots:=array:

																							#Background diamond texture
																							overlay_slot:
																								VerticalAlignment:=vertical_alignment.Center
																								HorizontalAlignment:=horizontal_alignment.Center
																								#Distribution:=option{1.0}
																								Padding:=margin{Left:=0.0, Top:=0.0, Right:=0.0, Bottom:=0.0}
																								Widget:=texture_block{DefaultImage:=Textures.General.T_GenArrow, DefaultDesiredSize:=vector2{X:=56.0, Y:=36.0}, DefaultTint:=MakeColorFromHex("#FFFFFFFF")}

																							#Inner diamond texture
																							overlay_slot:
																								VerticalAlignment:=vertical_alignment.Center
																								HorizontalAlignment:=horizontal_alignment.Center
																								#Distribution:=option{1.0}
																								Padding:=margin{Left:=1.0, Top:=0.0, Right:=0.0, Bottom:=0.0}
																								Widget:=texture_block{DefaultImage:=Textures.General.T_GenArrow, DefaultDesiredSize:=vector2{X:=48.0, Y:=30.0}, DefaultTint:=MakeColorFromHex("#ba8d8eFF")}

																							#Bed icon texture
																							overlay_slot:
																								VerticalAlignment:=vertical_alignment.Center
																								HorizontalAlignment:=horizontal_alignment.Center
																								#Distribution:=option{1.0}
																								Padding:=margin{Left:=3.0, Top:=0.0, Right:=0.0, Bottom:=0.0}
																								Widget:=texture_block{DefaultImage:=Textures.General.T_UI_Icon_bed_right, DefaultDesiredSize:=vector2{X:=24.0, Y:=24.0}, DefaultTint:=MakeColorFromHex("#FFFFFFFF")}
																				
																				#Small room arrow - Red
																				stack_box_slot:
																					VerticalAlignment:=vertical_alignment.Fill
																					HorizontalAlignment:=horizontal_alignment.Fill
																					#Distribution:=option{1.0}
																					Padding:=margin{Left:=1.0, Top:=0.0, Right:=7.0, Bottom:=0.0}
																					Widget:=overlay:
																						Slots:=array:

																							#Background diamond texture
																							overlay_slot:
																								VerticalAlignment:=vertical_alignment.Center
																								HorizontalAlignment:=horizontal_alignment.Center
																								#Distribution:=option{1.0}
																								Padding:=margin{Left:=0.0, Top:=0.0, Right:=0.0, Bottom:=0.0}
																								Widget:=texture_block{DefaultImage:=Textures.General.T_GenArrow, DefaultDesiredSize:=vector2{X:=56.0, Y:=36.0}, DefaultTint:=MakeColorFromHex("#FFFFFFFF")}

																							#Inner diamond texture
																							overlay_slot:
																								VerticalAlignment:=vertical_alignment.Center
																								HorizontalAlignment:=horizontal_alignment.Center
																								#Distribution:=option{1.0}
																								Padding:=margin{Left:=1.0, Top:=0.0, Right:=0.0, Bottom:=0.0}
																								Widget:=texture_block{DefaultImage:=Textures.General.T_GenArrow, DefaultDesiredSize:=vector2{X:=48.0, Y:=30.0}, DefaultTint:=MakeColorFromHex("#741a1dFF")}

																							#Bed icon texture
																							overlay_slot:
																								VerticalAlignment:=vertical_alignment.Center
																								HorizontalAlignment:=horizontal_alignment.Center
																								#Distribution:=option{1.0}
																								Padding:=margin{Left:=3.0, Top:=0.0, Right:=0.0, Bottom:=0.0}
																								Widget:=texture_block{DefaultImage:=Textures.General.T_UI_Icon_bed_right, DefaultDesiredSize:=vector2{X:=24.0, Y:=24.0}, DefaultTint:=MakeColorFromHex("#FFFFFFFF")}

																				#Small room arrow - Red
																				stack_box_slot:
																					VerticalAlignment:=vertical_alignment.Fill
																					HorizontalAlignment:=horizontal_alignment.Fill
																					#Distribution:=option{1.0}
																					Padding:=margin{Left:=1.0, Top:=0.0, Right:=7.0, Bottom:=0.0}
																					Widget:=overlay:
																						Slots:=array:

																							#Background diamond texture
																							overlay_slot:
																								VerticalAlignment:=vertical_alignment.Center
																								HorizontalAlignment:=horizontal_alignment.Center
																								#Distribution:=option{1.0}
																								Padding:=margin{Left:=0.0, Top:=0.0, Right:=0.0, Bottom:=0.0}
																								Widget:=texture_block{DefaultImage:=Textures.General.T_GenArrow, DefaultDesiredSize:=vector2{X:=56.0, Y:=36.0}, DefaultTint:=MakeColorFromHex("#FFFFFFFF")}

																							#Inner diamond texture
																							overlay_slot:
																								VerticalAlignment:=vertical_alignment.Center
																								HorizontalAlignment:=horizontal_alignment.Center
																								#Distribution:=option{1.0}
																								Padding:=margin{Left:=1.0, Top:=0.0, Right:=0.0, Bottom:=0.0}
																								Widget:=texture_block{DefaultImage:=Textures.General.T_GenArrow, DefaultDesiredSize:=vector2{X:=48.0, Y:=30.0}, DefaultTint:=MakeColorFromHex("#741a1dFF")}

																							#Bed icon texture
																							overlay_slot:
																								VerticalAlignment:=vertical_alignment.Center
																								HorizontalAlignment:=horizontal_alignment.Center
																								#Distribution:=option{1.0}
																								Padding:=margin{Left:=3.0, Top:=0.0, Right:=0.0, Bottom:=0.0}
																								Widget:=texture_block{DefaultImage:=Textures.General.T_UI_Icon_bed_right, DefaultDesiredSize:=vector2{X:=24.0, Y:=24.0}, DefaultTint:=MakeColorFromHex("#FFFFFFFF")}

																				#Small room arrow - Red
																				stack_box_slot:
																					VerticalAlignment:=vertical_alignment.Fill
																					HorizontalAlignment:=horizontal_alignment.Fill
																					#Distribution:=option{1.0}
																					Padding:=margin{Left:=1.0, Top:=0.0, Right:=7.0, Bottom:=0.0}
																					Widget:=overlay:
																						Slots:=array:

																							#Background diamond texture
																							overlay_slot:
																								VerticalAlignment:=vertical_alignment.Center
																								HorizontalAlignment:=horizontal_alignment.Center
																								#Distribution:=option{1.0}
																								Padding:=margin{Left:=0.0, Top:=0.0, Right:=0.0, Bottom:=0.0}
																								Widget:=texture_block{DefaultImage:=Textures.General.T_GenArrow, DefaultDesiredSize:=vector2{X:=56.0, Y:=36.0}, DefaultTint:=MakeColorFromHex("#FFFFFFFF")}

																							#Inner diamond texture
																							overlay_slot:
																								VerticalAlignment:=vertical_alignment.Center
																								HorizontalAlignment:=horizontal_alignment.Center
																								#Distribution:=option{1.0}
																								Padding:=margin{Left:=1.0, Top:=0.0, Right:=0.0, Bottom:=0.0}
																								Widget:=texture_block{DefaultImage:=Textures.General.T_GenArrow, DefaultDesiredSize:=vector2{X:=48.0, Y:=30.0}, DefaultTint:=MakeColorFromHex("#741a1dFF")}

																							#Bed icon texture
																							overlay_slot:
																								VerticalAlignment:=vertical_alignment.Center
																								HorizontalAlignment:=horizontal_alignment.Center
																								#Distribution:=option{1.0}
																								Padding:=margin{Left:=3.0, Top:=0.0, Right:=0.0, Bottom:=0.0}
																								Widget:=texture_block{DefaultImage:=Textures.General.T_UI_Icon_bed_right, DefaultDesiredSize:=vector2{X:=24.0, Y:=24.0}, DefaultTint:=MakeColorFromHex("#FFFFFFFF")}

																				#Big room module - Red
																				stack_box_slot:
																					VerticalAlignment:=vertical_alignment.Fill
																					HorizontalAlignment:=horizontal_alignment.Fill
																					#Distribution:=option{1.0}
																					Padding:=margin{Left:=0.0, Top:=0.0, Right:=7.0, Bottom:=0.0}
																					Widget:=overlay:
																						Slots:=array:

																							#Big Slot
																							overlay_slot:
																								VerticalAlignment:=vertical_alignment.Center
																								HorizontalAlignment:=horizontal_alignment.Left
																								#Distribution:=option{1.0}
																								Padding:=margin{Left:=0.0, Top:=8.0, Right:=0.0, Bottom:=0.0}
																								Widget:=overlay:
																									Slots:=array:
																										
																										#outher diamond texture
																										overlay_slot:
																											VerticalAlignment:=vertical_alignment.Center
																											HorizontalAlignment:=horizontal_alignment.Center
																											#Distribution:=option{1.0}
																											Padding:=margin{Left:=0.0, Top:=0.0, Right:=0.0, Bottom:=0.0}
																											Widget:=texture_block{DefaultImage:=Textures.General.T_GenTooltip, DefaultDesiredSize:=vector2{X:=62.0, Y:=62.0}, DefaultTint:=MakeColorFromHex("#FFFFFFFF")}

																										#Inner diamond texture
																										overlay_slot:
																											VerticalAlignment:=vertical_alignment.Center
																											HorizontalAlignment:=horizontal_alignment.Center
																											#Distribution:=option{1.0}
																											Padding:=margin{Left:=0.0, Top:=1.0, Right:=0.0, Bottom:=2.0}
																											Widget:=texture_block{DefaultImage:=Textures.General.T_GenTooltip, DefaultDesiredSize:=vector2{X:=54.0, Y:=52.0}, DefaultTint:=MakeColorFromHex("#db242aFF")}

																										#Bed icon texture
																										overlay_slot:
																											VerticalAlignment:=vertical_alignment.Top
																											HorizontalAlignment:=horizontal_alignment.Center
																											#Distribution:=option{1.0}
																											Padding:=margin{Left:=0.0, Top:=10.0, Right:=0.0, Bottom:=0.0}
																											Widget:=texture_block{DefaultImage:=Textures.General.T_UI_Icon_bed_right, DefaultDesiredSize:=vector2{X:=30.0, Y:=30.0}, DefaultTint:=MakeColorFromHex("#FFFFFFFF")}

														#Header - room number
														stack_box_slot:
															VerticalAlignment:=vertical_alignment.Center
															HorizontalAlignment:=horizontal_alignment.Center
															#Distribution:=option{1.0}
															Padding:=margin{Left:=35.0, Top:=0.0, Right:=35.0, Bottom:=10.0}
															Widget:=text_block{DefaultText := StringToMessage("Room 1"), DefaultTextColor:=MakeColorFromHex("#FFFFFFFF"), DefaultShadowColor:=MakeColorFromHex("#000000FF"), DefaultShadowOffset:=option{vector2{X:=3.0, Y:=3.0}}}
														
														#Blue team beds
														stack_box_slot:
															VerticalAlignment:=vertical_alignment.Center
															HorizontalAlignment:=horizontal_alignment.Right
															#Distribution:=option{1.0}
															Padding:=margin{Left:=0.0, Top:=0.0, Right:=0.0, Bottom:=0.0}
															Widget:=overlay:
																Slots:=array:
																	
																	#Makeshift size box for first team beds
																	overlay_slot:
																		VerticalAlignment:=vertical_alignment.Center
																		HorizontalAlignment:=horizontal_alignment.Center
																		#Distribution:=option{1.0}
																		Padding:=margin{Left:=0.0, Top:=0.0, Right:=0.0, Bottom:=0.0}
																		Widget:=texture_block{DefaultImage:=Textures.General.T_GenSquTrans, DefaultDesiredSize:=vector2{X:=315.0, Y:=65.0}, DefaultTint:=MakeColorFromHex("#0000FFFF")}

																	#Actual content for first team beds
																	overlay_slot:
																		VerticalAlignment:=vertical_alignment.Center
																		HorizontalAlignment:=horizontal_alignment.Right
																		#Distribution:=option{1.0}
																		Padding:=margin{Left:=0.0, Top:=0.0, Right:=0.0, Bottom:=10.0}
																		Widget:=stack_box:
																			Orientation:=orientation.Horizontal
																			Slots:=array:

																				#Big room module - Blue
																				stack_box_slot:
																					VerticalAlignment:=vertical_alignment.Fill
																					HorizontalAlignment:=horizontal_alignment.Fill
																					#Distribution:=option{1.0}
																					Padding:=margin{Left:=0.0, Top:=0.0, Right:=7.0, Bottom:=0.0}
																					Widget:=overlay:
																						Slots:=array:

																							#Big Slot
																							overlay_slot:
																								VerticalAlignment:=vertical_alignment.Center
																								HorizontalAlignment:=horizontal_alignment.Left
																								#Distribution:=option{1.0}
																								Padding:=margin{Left:=0.0, Top:=8.0, Right:=0.0, Bottom:=0.0}
																								Widget:=overlay:
																									Slots:=array:
																										
																										#outher diamond texture
																										overlay_slot:
																											VerticalAlignment:=vertical_alignment.Center
																											HorizontalAlignment:=horizontal_alignment.Center
																											#Distribution:=option{1.0}
																											Padding:=margin{Left:=0.0, Top:=0.0, Right:=0.0, Bottom:=0.0}
																											Widget:=texture_block{DefaultImage:=Textures.General.T_GenTooltip, DefaultDesiredSize:=vector2{X:=62.0, Y:=62.0}, DefaultTint:=MakeColorFromHex("#FFFFFFFF")}

																										#Inner diamond texture
																										overlay_slot:
																											VerticalAlignment:=vertical_alignment.Center
																											HorizontalAlignment:=horizontal_alignment.Center
																											#Distribution:=option{1.0}
																											Padding:=margin{Left:=0.0, Top:=1.0, Right:=0.0, Bottom:=2.0}
																											Widget:=texture_block{DefaultImage:=Textures.General.T_GenTooltip, DefaultDesiredSize:=vector2{X:=54.0, Y:=52.0}, DefaultTint:=MakeColorFromHex("#417af4FF")}

																										#Bed icon texture
																										overlay_slot:
																											VerticalAlignment:=vertical_alignment.Top
																											HorizontalAlignment:=horizontal_alignment.Center
																											#Distribution:=option{1.0}
																											Padding:=margin{Left:=0.0, Top:=10.0, Right:=3.0, Bottom:=0.0}
																											Widget:=texture_block{DefaultImage:=Textures.General.T_UI_Icon_bed, DefaultDesiredSize:=vector2{X:=30.0, Y:=30.0}, DefaultTint:=MakeColorFromHex("#FFFFFFFF")}
																				
																				#Small room arrow - red highlight
																				stack_box_slot:
																					VerticalAlignment:=vertical_alignment.Fill
																					HorizontalAlignment:=horizontal_alignment.Fill
																					#Distribution:=option{1.0}
																					Padding:=margin{Left:=0.0, Top:=0.0, Right:=7.0, Bottom:=0.0}
																					Widget:=overlay:
																						Slots:=array:

																							#Background diamond texture
																							overlay_slot:
																								VerticalAlignment:=vertical_alignment.Center
																								HorizontalAlignment:=horizontal_alignment.Center
																								#Distribution:=option{1.0}
																								Padding:=margin{Left:=0.0, Top:=0.0, Right:=0.0, Bottom:=0.0}
																								Widget:=texture_block{DefaultImage:=Textures.General.T_GenArrowLeft, DefaultDesiredSize:=vector2{X:=56.0, Y:=36.0}, DefaultTint:=MakeColorFromHex("#FFFFFFFF")}

																							#Inner diamond texture
																							overlay_slot:
																								VerticalAlignment:=vertical_alignment.Center
																								HorizontalAlignment:=horizontal_alignment.Center
																								#Distribution:=option{1.0}
																								Padding:=margin{Left:=1.0, Top:=0.0, Right:=2.0, Bottom:=0.0}
																								Widget:=texture_block{DefaultImage:=Textures.General.T_GenArrowLeft, DefaultDesiredSize:=vector2{X:=48.0, Y:=30.0}, DefaultTint:=MakeColorFromHex("#1e3a76FF")}

																							#Bed icon texture
																							overlay_slot:
																								VerticalAlignment:=vertical_alignment.Center
																								HorizontalAlignment:=horizontal_alignment.Center
																								#Distribution:=option{1.0}
																								Padding:=margin{Left:=0.0, Top:=0.0, Right:=3.0, Bottom:=0.0}
																								Widget:=texture_block{DefaultImage:=Textures.General.T_UI_Icon_bed, DefaultDesiredSize:=vector2{X:=24.0, Y:=24.0}, DefaultTint:=MakeColorFromHex("#FFFFFFFF")}
																																								
																				#Small room arrow - Red
																				stack_box_slot:
																					VerticalAlignment:=vertical_alignment.Fill
																					HorizontalAlignment:=horizontal_alignment.Fill
																					#Distribution:=option{1.0}
																					Padding:=margin{Left:=1.0, Top:=0.0, Right:=7.0, Bottom:=0.0}
																					Widget:=overlay:
																						Slots:=array:

																							#Background diamond texture
																							overlay_slot:
																								VerticalAlignment:=vertical_alignment.Center
																								HorizontalAlignment:=horizontal_alignment.Center
																								#Distribution:=option{1.0}
																								Padding:=margin{Left:=0.0, Top:=0.0, Right:=0.0, Bottom:=0.0}
																								Widget:=texture_block{DefaultImage:=Textures.General.T_GenArrowLeft, DefaultDesiredSize:=vector2{X:=56.0, Y:=36.0}, DefaultTint:=MakeColorFromHex("#FFFFFFFF")}

																							#Inner diamond texture
																							overlay_slot:
																								VerticalAlignment:=vertical_alignment.Center
																								HorizontalAlignment:=horizontal_alignment.Center
																								#Distribution:=option{1.0}
																								Padding:=margin{Left:=1.0, Top:=0.0, Right:=2.0, Bottom:=0.0}
																								Widget:=texture_block{DefaultImage:=Textures.General.T_GenArrowLeft, DefaultDesiredSize:=vector2{X:=48.0, Y:=30.0}, DefaultTint:=MakeColorFromHex("#1e3a76FF")}

																							#Bed icon texture
																							overlay_slot:
																								VerticalAlignment:=vertical_alignment.Center
																								HorizontalAlignment:=horizontal_alignment.Center
																								#Distribution:=option{1.0}
																								Padding:=margin{Left:=0.0, Top:=0.0, Right:=3.0, Bottom:=0.0}
																								Widget:=texture_block{DefaultImage:=Textures.General.T_UI_Icon_bed, DefaultDesiredSize:=vector2{X:=24.0, Y:=24.0}, DefaultTint:=MakeColorFromHex("#FFFFFFFF")}

																				#Small room arrow - Red
																				stack_box_slot:
																					VerticalAlignment:=vertical_alignment.Fill
																					HorizontalAlignment:=horizontal_alignment.Fill
																					#Distribution:=option{1.0}
																					Padding:=margin{Left:=1.0, Top:=0.0, Right:=7.0, Bottom:=0.0}
																					Widget:=overlay:
																						Slots:=array:

																							#Background diamond texture
																							overlay_slot:
																								VerticalAlignment:=vertical_alignment.Center
																								HorizontalAlignment:=horizontal_alignment.Center
																								#Distribution:=option{1.0}
																								Padding:=margin{Left:=0.0, Top:=0.0, Right:=0.0, Bottom:=0.0}
																								Widget:=texture_block{DefaultImage:=Textures.General.T_GenArrowLeft, DefaultDesiredSize:=vector2{X:=56.0, Y:=36.0}, DefaultTint:=MakeColorFromHex("#FFFFFFFF")}

																							#Inner diamond texture
																							overlay_slot:
																								VerticalAlignment:=vertical_alignment.Center
																								HorizontalAlignment:=horizontal_alignment.Center
																								#Distribution:=option{1.0}
																								Padding:=margin{Left:=1.0, Top:=0.0, Right:=2.0, Bottom:=0.0}
																								Widget:=texture_block{DefaultImage:=Textures.General.T_GenArrowLeft, DefaultDesiredSize:=vector2{X:=48.0, Y:=30.0}, DefaultTint:=MakeColorFromHex("#1e3a76FF")}

																							#Bed icon texture
																							overlay_slot:
																								VerticalAlignment:=vertical_alignment.Center
																								HorizontalAlignment:=horizontal_alignment.Center
																								#Distribution:=option{1.0}
																								Padding:=margin{Left:=0.0, Top:=0.0, Right:=3.0, Bottom:=0.0}
																								Widget:=texture_block{DefaultImage:=Textures.General.T_UI_Icon_bed, DefaultDesiredSize:=vector2{X:=24.0, Y:=24.0}, DefaultTint:=MakeColorFromHex("#FFFFFFFF")}

																				#Small room arrow - Red
																				stack_box_slot:
																					VerticalAlignment:=vertical_alignment.Fill
																					HorizontalAlignment:=horizontal_alignment.Fill
																					#Distribution:=option{1.0}
																					Padding:=margin{Left:=1.0, Top:=0.0, Right:=7.0, Bottom:=0.0}
																					Widget:=overlay:
																						Slots:=array:

																							#Background diamond texture
																							overlay_slot:
																								VerticalAlignment:=vertical_alignment.Center
																								HorizontalAlignment:=horizontal_alignment.Center
																								#Distribution:=option{1.0}
																								Padding:=margin{Left:=0.0, Top:=0.0, Right:=0.0, Bottom:=0.0}
																								Widget:=texture_block{DefaultImage:=Textures.General.T_GenArrowLeft, DefaultDesiredSize:=vector2{X:=56.0, Y:=36.0}, DefaultTint:=MakeColorFromHex("#FFFFFFFF")}

																							#Inner diamond texture
																							overlay_slot:
																								VerticalAlignment:=vertical_alignment.Center
																								HorizontalAlignment:=horizontal_alignment.Center
																								#Distribution:=option{1.0}
																								Padding:=margin{Left:=1.0, Top:=0.0, Right:=2.0, Bottom:=0.0}
																								Widget:=texture_block{DefaultImage:=Textures.General.T_GenArrowLeft, DefaultDesiredSize:=vector2{X:=48.0, Y:=30.0}, DefaultTint:=MakeColorFromHex("#1e3a76FF")}

																							#Bed icon texture
																							overlay_slot:
																								VerticalAlignment:=vertical_alignment.Center
																								HorizontalAlignment:=horizontal_alignment.Center
																								#Distribution:=option{1.0}
																								Padding:=margin{Left:=0.0, Top:=0.0, Right:=3.0, Bottom:=0.0}
																								Widget:=texture_block{DefaultImage:=Textures.General.T_UI_Icon_bed, DefaultDesiredSize:=vector2{X:=24.0, Y:=24.0}, DefaultTint:=MakeColorFromHex("#FFFFFFFF")}
											
											#Room layout indicator
											stack_box_slot:
												VerticalAlignment:=vertical_alignment.Center
												HorizontalAlignment:=horizontal_alignment.Center
												#Distribution:=option{1.0}
												Padding:=margin{Left:=0.0, Top:=0.0, Right:=5.0, Bottom:=15.0}
												Widget:=overlay:
													Slots:=array:
														
														#Makeshift Size box
														overlay_slot:
															VerticalAlignment:=vertical_alignment.Center
															HorizontalAlignment:=horizontal_alignment.Center
															#Distribution:=option{1.0}
															Padding:=margin{Left:=0.0, Top:=0.0, Right:=0.0, Bottom:=0.0}
															Widget:=texture_block{DefaultImage:=Textures.General.T_GenSquTrans, DefaultDesiredSize:=vector2{X:=350.0, Y:=35.0}, DefaultTint:=MakeColorFromHex("#00FF00FF")}

														#horizontal stack box for rooms layout
														overlay_slot:
															VerticalAlignment:=vertical_alignment.Center
															HorizontalAlignment:=horizontal_alignment.Center
															#Distribution:=option{1.0}
															Padding:=margin{Left:=0.0, Top:=0.0, Right:=0.0, Bottom:=0.0}
															Widget:=stack_box:
																Orientation:=orientation.Horizontal
																Slots:=array:

																	#Preroom slot, R
																	stack_box_slot:
																		VerticalAlignment:=vertical_alignment.Center
																		HorizontalAlignment:=horizontal_alignment.Center
																		Distribution:=option{1.0}
																		Padding:=margin{Left:=0.0, Top:=0.0, Right:=0.0, Bottom:=0.0}
																		Widget:=overlay:
																			Slots:=array:

																				overlay_slot:
																					VerticalAlignment:=vertical_alignment.Center
																					HorizontalAlignment:=horizontal_alignment.Center
																					#Distribution:=option{1.0}
																					Padding:=margin{Left:=0.0, Top:=0.0, Right:=0.0, Bottom:=0.0}
																					Widget:=texture_block{DefaultImage:=Textures.General.T_GenericSquare, DefaultDesiredSize:=vector2{X:=22.0, Y:=24.0}, DefaultTint:=MakeColorFromHex("#FFFFFFFF")}

																				overlay_slot:
																					VerticalAlignment:=vertical_alignment.Center
																					HorizontalAlignment:=horizontal_alignment.Center
																					#Distribution:=option{1.0}
																					Padding:=margin{Left:=0.0, Top:=0.0, Right:=0.0, Bottom:=0.0}
																					Widget:=text_block{DefaultText := StringToMessage("R"), DefaultTextColor:=MakeColorFromHex("#db242aFF"), DefaultShadowColor:=MakeColorFromHex("#000000FF"), DefaultShadowOffset:=option{vector2{X:=0.0, Y:=0.0}}}
																	
																	#Room slot, red
																	stack_box_slot:
																		VerticalAlignment:=vertical_alignment.Center
																		HorizontalAlignment:=horizontal_alignment.Center
																		Distribution:=option{1.0}
																		Padding:=margin{Left:=0.0, Top:=0.0, Right:=0.0, Bottom:=0.0}
																		Widget:=overlay:
																			Slots:=array:

																				overlay_slot:
																					VerticalAlignment:=vertical_alignment.Center
																					HorizontalAlignment:=horizontal_alignment.Center
																					#Distribution:=option{1.0}
																					Padding:=margin{Left:=0.0, Top:=0.0, Right:=0.0, Bottom:=0.0}
																					Widget:=texture_block{DefaultImage:=Textures.General.T_GenericSquare, DefaultDesiredSize:=vector2{X:=22.0, Y:=24.0}, DefaultTint:=MakeColorFromHex("#FFFFFFFF")}

																				overlay_slot:
																					VerticalAlignment:=vertical_alignment.Center
																					HorizontalAlignment:=horizontal_alignment.Center
																					#Distribution:=option{1.0}
																					Padding:=margin{Left:=0.0, Top:=0.0, Right:=0.0, Bottom:=0.0}
																					Widget:=texture_block{DefaultImage:=Textures.General.T_GenericSquare, DefaultDesiredSize:=vector2{X:=18.0, Y:=18.0}, DefaultTint:=MakeColorFromHex("#db242aFF")}

																	#Room slot, Grey
																	stack_box_slot:
																		VerticalAlignment:=vertical_alignment.Center
																		HorizontalAlignment:=horizontal_alignment.Center
																		Distribution:=option{1.0}
																		Padding:=margin{Left:=0.0, Top:=0.0, Right:=0.0, Bottom:=0.0}
																		Widget:=overlay:
																			Slots:=array:

																				overlay_slot:
																					VerticalAlignment:=vertical_alignment.Center
																					HorizontalAlignment:=horizontal_alignment.Center
																					#Distribution:=option{1.0}
																					Padding:=margin{Left:=0.0, Top:=0.0, Right:=0.0, Bottom:=0.0}
																					Widget:=texture_block{DefaultImage:=Textures.General.T_GenericSquare, DefaultDesiredSize:=vector2{X:=22.0, Y:=24.0}, DefaultTint:=MakeColorFromHex("#FFFFFFFF")}

																				overlay_slot:
																					VerticalAlignment:=vertical_alignment.Center
																					HorizontalAlignment:=horizontal_alignment.Center
																					#Distribution:=option{1.0}
																					Padding:=margin{Left:=0.0, Top:=0.0, Right:=0.0, Bottom:=0.0}
																					Widget:=texture_block{DefaultImage:=Textures.General.T_GenericSquare, DefaultDesiredSize:=vector2{X:=18.0, Y:=18.0}, DefaultTint:=MakeColorFromHex("#c5c6c8FF")}

																	#Room slot, Grey
																	stack_box_slot:
																		VerticalAlignment:=vertical_alignment.Center
																		HorizontalAlignment:=horizontal_alignment.Center
																		Distribution:=option{1.0}
																		Padding:=margin{Left:=0.0, Top:=0.0, Right:=0.0, Bottom:=0.0}
																		Widget:=overlay:
																			Slots:=array:

																				overlay_slot:
																					VerticalAlignment:=vertical_alignment.Center
																					HorizontalAlignment:=horizontal_alignment.Center
																					#Distribution:=option{1.0}
																					Padding:=margin{Left:=0.0, Top:=0.0, Right:=0.0, Bottom:=0.0}
																					Widget:=texture_block{DefaultImage:=Textures.General.T_GenericSquare, DefaultDesiredSize:=vector2{X:=22.0, Y:=24.0}, DefaultTint:=MakeColorFromHex("#FFFFFFFF")}

																				overlay_slot:
																					VerticalAlignment:=vertical_alignment.Center
																					HorizontalAlignment:=horizontal_alignment.Center
																					#Distribution:=option{1.0}
																					Padding:=margin{Left:=0.0, Top:=0.0, Right:=0.0, Bottom:=0.0}
																					Widget:=texture_block{DefaultImage:=Textures.General.T_GenericSquare, DefaultDesiredSize:=vector2{X:=18.0, Y:=18.0}, DefaultTint:=MakeColorFromHex("#c5c6c8FF")}

																	#Room slot, Grey
																	stack_box_slot:
																		VerticalAlignment:=vertical_alignment.Center
																		HorizontalAlignment:=horizontal_alignment.Center
																		Distribution:=option{1.0}
																		Padding:=margin{Left:=0.0, Top:=0.0, Right:=0.0, Bottom:=0.0}
																		Widget:=overlay:
																			Slots:=array:

																				overlay_slot:
																					VerticalAlignment:=vertical_alignment.Center
																					HorizontalAlignment:=horizontal_alignment.Center
																					#Distribution:=option{1.0}
																					Padding:=margin{Left:=0.0, Top:=0.0, Right:=0.0, Bottom:=0.0}
																					Widget:=texture_block{DefaultImage:=Textures.General.T_GenericSquare, DefaultDesiredSize:=vector2{X:=22.0, Y:=24.0}, DefaultTint:=MakeColorFromHex("#FFFFFFFF")}

																				overlay_slot:
																					VerticalAlignment:=vertical_alignment.Center
																					HorizontalAlignment:=horizontal_alignment.Center
																					#Distribution:=option{1.0}
																					Padding:=margin{Left:=0.0, Top:=0.0, Right:=0.0, Bottom:=0.0}
																					Widget:=texture_block{DefaultImage:=Textures.General.T_GenericSquare, DefaultDesiredSize:=vector2{X:=18.0, Y:=18.0}, DefaultTint:=MakeColorFromHex("#c5c6c8FF")}

																	#Halfway room slot, Grey
																	stack_box_slot:
																		VerticalAlignment:=vertical_alignment.Center
																		HorizontalAlignment:=horizontal_alignment.Fill
																		#Distribution:=option{1.0}
																		Padding:=margin{Left:=0.0, Top:=0.0, Right:=0.0, Bottom:=0.0}
																		Widget:=overlay:
																			Slots:=array:

																				overlay_slot:
																					VerticalAlignment:=vertical_alignment.Center
																					HorizontalAlignment:=horizontal_alignment.Center
																					#Distribution:=option{1.0}
																					Padding:=margin{Left:=0.0, Top:=0.0, Right:=0.0, Bottom:=0.0}
																					Widget:=texture_block{DefaultImage:=Textures.General.T_GenericSquare, DefaultDesiredSize:=vector2{X:=50.0, Y:=30.0}, DefaultTint:=MakeColorFromHex("#FFFFFFFF")}

																				overlay_slot:
																					VerticalAlignment:=vertical_alignment.Center
																					HorizontalAlignment:=horizontal_alignment.Center
																					#Distribution:=option{1.0}
																					Padding:=margin{Left:=0.0, Top:=0.0, Right:=0.0, Bottom:=0.0}
																					Widget:=texture_block{DefaultImage:=Textures.General.T_GenericSquare, DefaultDesiredSize:=vector2{X:=42.0, Y:=22.0}, DefaultTint:=MakeColorFromHex("#c5c6c8FF")}

																	#Room slot, Grey
																	stack_box_slot:
																		VerticalAlignment:=vertical_alignment.Center
																		HorizontalAlignment:=horizontal_alignment.Center
																		Distribution:=option{1.0}
																		Padding:=margin{Left:=0.0, Top:=0.0, Right:=0.0, Bottom:=0.0}
																		Widget:=overlay:
																			Slots:=array:

																				overlay_slot:
																					VerticalAlignment:=vertical_alignment.Center
																					HorizontalAlignment:=horizontal_alignment.Center
																					#Distribution:=option{1.0}
																					Padding:=margin{Left:=0.0, Top:=0.0, Right:=0.0, Bottom:=0.0}
																					Widget:=texture_block{DefaultImage:=Textures.General.T_GenericSquare, DefaultDesiredSize:=vector2{X:=22.0, Y:=24.0}, DefaultTint:=MakeColorFromHex("#FFFFFFFF")}

																				overlay_slot:
																					VerticalAlignment:=vertical_alignment.Center
																					HorizontalAlignment:=horizontal_alignment.Center
																					#Distribution:=option{1.0}
																					Padding:=margin{Left:=0.0, Top:=0.0, Right:=0.0, Bottom:=0.0}
																					Widget:=texture_block{DefaultImage:=Textures.General.T_GenericSquare, DefaultDesiredSize:=vector2{X:=18.0, Y:=18.0}, DefaultTint:=MakeColorFromHex("#c5c6c8FF")}

																	#Room slot, Blue
																	stack_box_slot:
																		VerticalAlignment:=vertical_alignment.Center
																		HorizontalAlignment:=horizontal_alignment.Center
																		Distribution:=option{1.0}
																		Padding:=margin{Left:=0.0, Top:=0.0, Right:=0.0, Bottom:=0.0}
																		Widget:=overlay:
																			Slots:=array:

																				overlay_slot:
																					VerticalAlignment:=vertical_alignment.Center
																					HorizontalAlignment:=horizontal_alignment.Center
																					#Distribution:=option{1.0}
																					Padding:=margin{Left:=0.0, Top:=0.0, Right:=0.0, Bottom:=0.0}
																					Widget:=texture_block{DefaultImage:=Textures.General.T_GenericSquare, DefaultDesiredSize:=vector2{X:=22.0, Y:=24.0}, DefaultTint:=MakeColorFromHex("#FFFFFFFF")}

																				overlay_slot:
																					VerticalAlignment:=vertical_alignment.Center
																					HorizontalAlignment:=horizontal_alignment.Center
																					#Distribution:=option{1.0}
																					Padding:=margin{Left:=0.0, Top:=0.0, Right:=0.0, Bottom:=0.0}
																					Widget:=texture_block{DefaultImage:=Textures.General.T_GenericSquare, DefaultDesiredSize:=vector2{X:=18.0, Y:=18.0}, DefaultTint:=MakeColorFromHex("#417af4FF")}

																	#Room slot, Blue
																	stack_box_slot:
																		VerticalAlignment:=vertical_alignment.Center
																		HorizontalAlignment:=horizontal_alignment.Center
																		Distribution:=option{1.0}
																		Padding:=margin{Left:=0.0, Top:=0.0, Right:=0.0, Bottom:=0.0}
																		Widget:=overlay:
																			Slots:=array:

																				overlay_slot:
																					VerticalAlignment:=vertical_alignment.Center
																					HorizontalAlignment:=horizontal_alignment.Center
																					#Distribution:=option{1.0}
																					Padding:=margin{Left:=0.0, Top:=0.0, Right:=0.0, Bottom:=0.0}
																					Widget:=texture_block{DefaultImage:=Textures.General.T_GenericSquare, DefaultDesiredSize:=vector2{X:=22.0, Y:=24.0}, DefaultTint:=MakeColorFromHex("#FFFFFFFF")}

																				overlay_slot:
																					VerticalAlignment:=vertical_alignment.Center
																					HorizontalAlignment:=horizontal_alignment.Center
																					#Distribution:=option{1.0}
																					Padding:=margin{Left:=0.0, Top:=0.0, Right:=0.0, Bottom:=0.0}
																					Widget:=texture_block{DefaultImage:=Textures.General.T_GenericSquare, DefaultDesiredSize:=vector2{X:=18.0, Y:=18.0}, DefaultTint:=MakeColorFromHex("#417af4FF")}

																	#Room slot, Blue
																	stack_box_slot:
																		VerticalAlignment:=vertical_alignment.Center
																		HorizontalAlignment:=horizontal_alignment.Center
																		Distribution:=option{1.0}
																		Padding:=margin{Left:=0.0, Top:=0.0, Right:=0.0, Bottom:=0.0}
																		Widget:=overlay:
																			Slots:=array:

																				overlay_slot:
																					VerticalAlignment:=vertical_alignment.Center
																					HorizontalAlignment:=horizontal_alignment.Center
																					#Distribution:=option{1.0}
																					Padding:=margin{Left:=0.0, Top:=0.0, Right:=0.0, Bottom:=0.0}
																					Widget:=texture_block{DefaultImage:=Textures.General.T_GenericSquare, DefaultDesiredSize:=vector2{X:=22.0, Y:=24.0}, DefaultTint:=MakeColorFromHex("#FFFFFFFF")}

																				overlay_slot:
																					VerticalAlignment:=vertical_alignment.Center
																					HorizontalAlignment:=horizontal_alignment.Center
																					#Distribution:=option{1.0}
																					Padding:=margin{Left:=0.0, Top:=0.0, Right:=0.0, Bottom:=0.0}
																					Widget:=texture_block{DefaultImage:=Textures.General.T_GenericSquare, DefaultDesiredSize:=vector2{X:=18.0, Y:=18.0}, DefaultTint:=MakeColorFromHex("#417af4FF")}

																	#Preroom slot, B
																	stack_box_slot:
																		VerticalAlignment:=vertical_alignment.Center
																		HorizontalAlignment:=horizontal_alignment.Center
																		Distribution:=option{1.0}
																		Padding:=margin{Left:=0.0, Top:=0.0, Right:=0.0, Bottom:=0.0}
																		Widget:=overlay:
																			Slots:=array:

																				overlay_slot:
																					VerticalAlignment:=vertical_alignment.Center
																					HorizontalAlignment:=horizontal_alignment.Center
																					#Distribution:=option{1.0}
																					Padding:=margin{Left:=0.0, Top:=0.0, Right:=0.0, Bottom:=0.0}
																					Widget:=texture_block{DefaultImage:=Textures.General.T_GenericSquare, DefaultDesiredSize:=vector2{X:=22.0, Y:=24.0}, DefaultTint:=MakeColorFromHex("#FFFFFFFF")}

																				overlay_slot:
																					VerticalAlignment:=vertical_alignment.Center
																					HorizontalAlignment:=horizontal_alignment.Center
																					#Distribution:=option{1.0}
																					Padding:=margin{Left:=0.0, Top:=0.0, Right:=0.0, Bottom:=0.0}
																					Widget:=text_block{DefaultText := StringToMessage("B"), DefaultTextColor:=MakeColorFromHex("#417af4FF"), DefaultShadowColor:=MakeColorFromHex("#000000FF"), DefaultShadowOffset:=option{vector2{X:=0.0, Y:=0.0}}}
													
			#Build of the screen
			HeaderCanvas:canvas = canvas:
				Slots:=array:
					#Defines the parameters for the canvas' slot
					canvas_slot:
						Anchors := anchors{Minimum := vector2{X := 0.227, Y := 0.15}, Maximum := vector2{X := 0.773, Y := 0.17}} # Anchors := anchors{Minimum := vector2{X := 0.227, Y := 0.0}, ...
						Alignment := vector2{X := 0.5, Y := 0.7} # Alignment := vector2{X := 0.5, Y := 0.5} ...
						Widget:=overlay:
							Slots:=array:
								overlay_slot:
									VerticalAlignment:=vertical_alignment.Top
									HorizontalAlignment:=horizontal_alignment.Center
									Padding:=margin{Left:=0.0, Top:=0.0, Right:=0.0, Bottom:=0.0}
									#Screen background
									Widget:=texture_block{DefaultImage:=Textures.General.T_GenSquTrans, DefaultDesiredSize:=vector2{X:=1810.0, Y:=170.0}}
							
								overlay_slot:
									VerticalAlignment:=vertical_alignment.Center
									HorizontalAlignment:=horizontal_alignment.Center
									Padding:=margin{Left:=0.0, Top:=0.0, Right:=0.0, Bottom:=0.0}
									#Screen background
									Widget:=MatchMarker
			
			if(PlayerUI:=GetPlayerUI[Player]):
				PlayerUI.AddWidget(HeaderCanvas)
				#>

