using { /Fortnite.com/Devices }
using { /Fortnite.com/UI }
using { /Verse.org/Simulation }
using { /Verse.org/Colors }
using { /Verse.org/Assets }
using { /Verse.org/Random }
using { /UnrealEngine.com/Temporary/UI }
using { /UnrealEngine.com/Temporary/SpatialMath}

ui_rewardable_data := class:
    Rewardable:rewardable
    RewardType:reward_type

ui_select_reward := class(creative_device):
    @editable
    TitleString:string = "CHOOSE ONE"

    @editable
    TitleTextColor:color = NamedColors.White

    @editable
    CardSize:vector2 = vector2{X:=427.0, Y:=854.0}

    @editable
    ItemNameTextColor:color = NamedColors.White

    @editable
    IconSize:vector2 = vector2{X:=250.0, Y:=250.0}

    @editable
    DescriptionTextColor:color = NamedColors.White

    @editable
    ButtonString:string = "ACCEPT"

    @editable
    DisplayTimeInSeconds:int = 30

    @editable
    TimerTextColor:color = NamedColors.White

    @editable
    TimerString:string = "Random selection in "    

    PlayerSelectionEvent:event(reward_selection_payload) = event(reward_selection_payload){}
    
    UIZOrder:type {_X:int where 0 < _X, _X <= 2147483647} = 1

    ShowUI(Player:player, Rewards:[]ui_rewardable_data, RewardableRoom:rewardableRoom): void =
        spawn {AwaitShow(Player, Rewards, RewardableRoom)}   

    AwaitShow<private>(Player:player, Rewards:[]ui_rewardable_data, RewardableRoom:rewardableRoom)<suspends>: void =
        if (Player, PlayerUI:=GetPlayerUI[Player]):           

            WidgetRemovedEvent:event() = event(){} # So all buttons stop listening when widget removed        

            CanvasAnswer:= ConstructCanvas(Player, Rewards, WidgetRemovedEvent, RewardableRoom)
            Canvas:=CanvasAnswer(0)
            TimerTextBlock:=CanvasAnswer(1)

            PlayerUI.AddWidget(Canvas, player_ui_slot{ZOrder:=UIZOrder, InputMode:=ui_input_mode.All})    
            
            AwaitShowing(Player, Rewards, TimerTextBlock, RewardableRoom)
            
            WidgetRemovedEvent.Signal()
            PlayerUI.RemoveWidget(Canvas)

    ConstructCanvas(Player:player,Rewards:[]ui_rewardable_data, WidgetRemovedEvent:event(), RewardableRoom:rewardableRoom): tuple(canvas, text_block) =         
        ItemStackBox:= stack_box:
            Orientation:=orientation.Horizontal
            Slots:=for (Reward:Rewards):
                # Reward Card
                stack_box_slot:
                    HorizontalAlignment:=horizontal_alignment.Center
                    VerticalAlignment:=vertical_alignment.Center
                    Padding:=margin{Left:=5.0, Right:=5.0}
                    Distribution:=option{1.0}
                    Widget:=ConstructRewardCard(Player, Reward,WidgetRemovedEvent, RewardableRoom)

        TimerTextBlock:=text_block{DefaultText:=StringToMessage(""), DefaultTextColor:=TimerTextColor}

        Canvas:=canvas:
            Slots:=array:
                canvas_slot:
                    Anchors:=anchors{Minimum:=vector2{X:=0.5,Y:=0.5},Maximum:=vector2{X:=0.5,Y:=0.5}}
                    Offsets:=margin{Top:=-435.0}
                    Alignment:=vector2{X:=0.5, Y:=1.0}
                    SizeToContent:=true
                    Widget:=text_block{DefaultText:=StringToMessage(TitleString), DefaultTextColor:=TitleTextColor}
                canvas_slot:
                    Anchors:=anchors{Minimum:=vector2{X:=0.5,Y:=0.5},Maximum:=vector2{X:=0.5,Y:=0.5}}
                    Offsets:=margin{Top:=-400.0}
                    Alignment:=vector2{X:=0.5, Y:=1.0}
                    SizeToContent:=true
                    Widget:=TimerTextBlock
                canvas_slot:
                    Anchors:=anchors{Minimum:=vector2{X:=0.5,Y:=0.5},Maximum:=vector2{X:=0.5,Y:=0.5}}
                    Alignment:=vector2{X:=0.5, Y:=0.5}
                    Offsets:=margin{Top:=-40.0}
                    SizeToContent:=true
                    Widget:=ItemStackBox   
        (Canvas, TimerTextBlock)

    ConstructRewardCard(Player:player, RewardableData:ui_rewardable_data, WidgetRemovedEvent:event(), RewardableRoom:rewardableRoom):widget=       
        SelectButton:button_loud = button_loud{DefaultText:=StringToMessage(ButtonString)}
        spawn {AwaitItemButtonSelection(SelectButton, RewardableData, WidgetRemovedEvent, RewardableRoom)} 
        Data := RewardableData.Rewardable

        InfoOverlay:=overlay{}

        if (Description:=describable[Data]):
            InfoOverlay.AddWidget(
                # Description
                overlay_slot:
                    VerticalAlignment:=vertical_alignment.Center
                    HorizontalAlignment:=horizontal_alignment.Left
                    Widget:=text_block{DefaultText:=StringToMessage(Description.GetDescription()), DefaultTextColor:=DescriptionTextColor,
                        DefaultJustification:=text_justification.Left}
            )

        if (Attributable:=attributable[Data], Attributes := Attributable.GetMaybeAttributes()?):
            InfoOverlay.AddWidget(
                    # Attributes
                overlay_slot:
                    VerticalAlignment:=vertical_alignment.Center
                    HorizontalAlignment:=horizontal_alignment.Left
                    Widget:=text_block{DefaultText:=StringToMessage(GetAttributeText(Attributes)), DefaultTextColor:=DescriptionTextColor,
                        DefaultJustification:=text_justification.Left}      
            )

        CardDataStackBox:=stack_box:
            Orientation:=orientation.Vertical
            Slots:=array:
                # Title
                stack_box_slot:
                    VerticalAlignment:=vertical_alignment.Center
                    HorizontalAlignment:=horizontal_alignment.Center
                    Padding:=margin{Top:=-20.0, Bottom:=0.0}
                    Widget:=text_block{DefaultText:=StringToMessage(Data.GetDisplayName()), DefaultTextColor:=ItemNameTextColor,
                        DefaultJustification:=text_justification.Center}
                # Icon
                stack_box_slot:
                    VerticalAlignment:=vertical_alignment.Center
                    HorizontalAlignment:=horizontal_alignment.Center
                    Padding:=margin{Top:=40.0, Bottom:=50.0}
                    Widget:=texture_block{DefaultImage:=GetIconTexture(Data), DefaultDesiredSize:=IconSize}
                # Info
                stack_box_slot:
                    VerticalAlignment:=vertical_alignment.Fill
                    HorizontalAlignment:=horizontal_alignment.Fill
                    Distribution:=option{1.0}
                    Widget:=InfoOverlay
            
        overlay:
            Slots:=array:
                # Background
                overlay_slot:
                    VerticalAlignment:=vertical_alignment.Center
                    HorizontalAlignment:=horizontal_alignment.Center
                    Widget:=texture_block{DefaultImage:=FindRarityCardBackgroundTexture(Data.GetRarity()), DefaultDesiredSize:=CardSize}
                # Data
                overlay_slot:
                    VerticalAlignment:=vertical_alignment.Fill
                    HorizontalAlignment:=horizontal_alignment.Fill
                    Padding:=margin{Top:=120.0, Left:=30.0, Right:=30.0, Bottom:=120.0}
                    Widget:=CardDataStackBox
                # Button
                overlay_slot:
                    VerticalAlignment:=vertical_alignment.Bottom
                    HorizontalAlignment:=horizontal_alignment.Center
                    Padding:=margin{Bottom:=40.0}
                    Widget:=SelectButton  

    AwaitShowing(Player:player, Rewards:[]ui_rewardable_data, TimerTextBlock:text_block, RewardableRoom:rewardableRoom)<suspends>: void =
        race:
            #AwaitHideEvent(Player)
            
            block:
                var SecondsLeft:int = DisplayTimeInSeconds
                loop:
                    TimerTextBlock.SetText(StringToMessage("{TimerString}{TimeInSecondsToString(SecondsLeft)}"))
                    Sleep(1.0)
                    set SecondsLeft -= 1

                    # No more time, select one randomly
                    if (SecondsLeft <= 0):
                        RandomIndex:=GetRandomInt(0, Rewards.Length-1)
                        if (Reward:=Rewards[RandomIndex]):
                            PlayerSelectionEvent.Signal(reward_selection_payload{Player:=Player, Reward:=Reward.Rewardable, RewardType:= Reward.RewardType})
                        break
            loop:
                ItemSelection := PlayerSelectionEvent.Await()
                if (Player = ItemSelection.Player):
                    break

    AwaitItemButtonSelection(Button:text_button_base, Reward:ui_rewardable_data, WidgetRemovedEvent:event(), RewardableRoom:rewardableRoom)<suspends>: void =
        race:
            block:
                WidgetRemovedEvent.Await()
                return
            block:
                WidgetMessage := Button.OnClick().Await() 
                PlayerSelectionEvent.Signal(reward_selection_payload{Player:=WidgetMessage.Player, Reward:=Reward.Rewardable, RewardType:= Reward.RewardType})
                IsRoomZero:=RewardableRoom.GetIsRoomZero()
                
                if (IsRoomZero = false):                 
                    #Remove the pending reward after the player has claimed it on any room that is not room Zero
                    RewardableRoom.RemovePlayerReward(WidgetMessage.Player)
                    

    GetIconTexture(Data:rewardable): texture=
        var NewTexture:texture = Textures.ItemIcons.RifleIcons.T_AssaultRifleB
        if (IconRewardable := iconTextureable[Data]):
            Print("Texture Name Found: {Data.GetDisplayName()}")
            set NewTexture = IconRewardable.GetIconTexture()
        else:
            Print("Texture Name Not Found: {Data.GetDisplayName()}")
        return NewTexture