using { /Fortnite.com/Devices }
using { /Fortnite.com/UI }
using { /Verse.org/Simulation }
using { /Verse.org/Colors }
using { /Verse.org/Assets }
using { /UnrealEngine.com/Temporary/Diagnostics }
using { /UnrealEngine.com/Temporary/UI }
using { /UnrealEngine.com/Temporary/SpatialMath}

preview_reward_ui_data := struct:
    Canvas:canvas
    RewardableRooms:rewardableRoom

<# 
UI shown to the player when an item reward must be selected
#>
ui_thumbnail_reward := class(creative_device):
    @editable
    TitleTextColor:color = NamedColors.White

    @editable
    ItemSize:vector2 = vector2{X:=80.0, Y:=80.0}

    @editable
    BlinkArrowFrequency:float = 0.8

    @editable # if < 0 means infinite
    NumBlinks:int = -1

    var UIPlayersData<private>:[player]preview_reward_ui_data = map{}

    HideWidgetEvent<public>:event(player) = event(player){}

    IsShowingToPlayer(Player:player)<decides><transacts>:void = UIPlayersData[Player]  

    Show(Player:player, Rewards:rewardableRoom): void =
        if (not UIPlayersData[Player]):
            AddWidget(Player, Rewards)           
        else:
            Print("You already have UI")
            Hide(Player)
            AddWidget(Player, Rewards)

    AddWidget(Player:player, Rewards:rewardableRoom): void=      
        if (PlayerUI:=GetPlayerUI[Player]):
            UIRewardableDatas :=  Rewards.GetUIRewardableData(Player)
            var Rewardables:[]rewardable = array{}
            for (UIRewardableData:UIRewardableDatas):
                set Rewardables += array{UIRewardableData.Rewardable}
            Canvas:=ConstructCanvas(Player,Rewardables)
            PlayerUI.AddWidget(Canvas)         
            if (set UIPlayersData[Player] = preview_reward_ui_data{Canvas:=Canvas,RewardableRooms:=Rewards}) {}

    Hide(Player:player): void =
        if (Canvas:=UIPlayersData[Player].Canvas, PlayerUI:=GetPlayerUI[Player]):
            PlayerUI.RemoveWidget(Canvas)
            HideWidgetEvent.Signal(Player)
            var NewMap:[player]preview_reward_ui_data = map{}
            for (KeyPlayer->Data : UIPlayersData, KeyPlayer<>Player):
                set NewMap = ConcatenateMaps(NewMap, map{KeyPlayer=>Data})
            set UIPlayersData = NewMap

    GetUIPlayerData(Player:player)<decides><transacts>:preview_reward_ui_data =
        UIPlayersData[Player]       

    ConstructCanvas(Player:player, Rewards:[]rewardable): canvas = 
        SectionsStackBox:=stack_box:
            Orientation:=orientation.Horizontal
            Slots:=for (Reward:Rewards):
                    # Reward Preview
                    stack_box_slot:
                        HorizontalAlignment:=horizontal_alignment.Center
                        VerticalAlignment:=vertical_alignment.Center
                        Padding:=margin{Left:=5.0, Right:=5.0}
                        Widget:=overlay:
                            Slots:=array:
                                overlay_slot:
                                    VerticalAlignment:=vertical_alignment.Center
                                    HorizontalAlignment:=horizontal_alignment.Center
                                    Widget:=texture_block{DefaultImage:=FindRarityThumbnailBackgroundTexture(Reward.GetRarity()), DefaultDesiredSize:=ItemSize}
                                overlay_slot:
                                    VerticalAlignment:=vertical_alignment.Center
                                    HorizontalAlignment:=horizontal_alignment.Center
                                    Widget:=texture_block{DefaultImage:=GetIconTexture(Reward), DefaultDesiredSize:=ItemSize}
     
        ArrowTexture:= texture_block{DefaultImage:=Textures.General.T_RewardIndicator, DefaultDesiredSize:=vector2{X:=550.0, Y:= 200.0}}
        spawn{BlinkRewardArrow(Player, ArrowTexture)}

        canvas:
            Slots:=array:
                canvas_slot:
                    Anchors:=anchors{Minimum:=vector2{X:=0.0,Y:=0.5},Maximum:=vector2{X:=0.0,Y:=0.5}}
                    Offsets:=margin{Left:=40.0, Top:=180.0}
                    Alignment:=vector2{X:=0.0, Y:=0.5}
                    SizeToContent:=true
                    Widget:=SectionsStackBox              
                canvas_slot:
                    Anchors:=anchors{Minimum:=vector2{X:=0.0,Y:=0.5},Maximum:=vector2{X:=0.0,Y:=0.5}}
                    Offsets:=margin{Left:=300.0, Top:=100.0}
                    Alignment:=vector2{X:=0.0, Y:=1.0}
                    SizeToContent:=true
                    Widget:=ArrowTexture

    BlinkRewardArrow(Player:player, ArrowTexture:texture_block)<suspends>: void =
        if (NumBlinks = 0):
            ArrowTexture.SetVisibility(widget_visibility.Hidden)
        else:
            var CurrentNumBlinks:int = 0

            race:
                loop:
                    ClosingWidgetPlayer:=HideWidgetEvent.Await()
                    if (ClosingWidgetPlayer = Player):
                        break

                loop:
                    ArrowTexture.SetVisibility(widget_visibility.Visible)
                    Sleep(BlinkArrowFrequency)
                    ArrowTexture.SetVisibility(widget_visibility.Hidden)

                    # Validate if not infinite
                    if (NumBlinks > 0):                   
                        set CurrentNumBlinks += 1
                        if (CurrentNumBlinks >= NumBlinks):
                            break

                    Sleep(BlinkArrowFrequency)

    GetIconTexture(Data:rewardable): texture=
        var NewTexture:texture = Textures.ItemIcons.RifleIcons.T_AssaultRifleB
        if (IconRewardable := iconTextureable[Data]):
            Print("Texture Name Found: {Data.GetDisplayName()}")
            set NewTexture = IconRewardable.GetIconTexture()
        else:
            Print("Texture Name Not Found: {Data.GetDisplayName()}")
        return NewTexture