using { /Fortnite.com/Devices }
using { /Fortnite.com/UI}
using { /Verse.org/Simulation }
using { /Verse.org/Colors }
using { /Verse.org/Colors/NamedColors }
using { /UnrealEngine.com/Temporary/UI }
using { /UnrealEngine.com/Temporary/SpatialMath }

ui_weapon_upgrade_device := class(creative_device):#class(ui_common)

    OnBegin<override>()<suspends>:void=
        CreateCanvas()

    CreateCanvas():void=
        Players:=GetPlayspace().GetPlayers()
        if(Player:=Players[0]):
            ImageBackground:=texture_block{DefaultImage:=Textures.General.T_Background}
            TitleTextWidget:=text_block{DefaultText := StringToMessage("WEAPON Rewards"), DefaultTextColor := NamedColors.White}
            BackButton:button_loud = button_loud{DefaultText:=StringToMessage("CLOSE")}

            #Definition for the header
            HeaderStackBox:=overlay:
                Slots:=array:
                    overlay_slot:
                        VerticalAlignment:=vertical_alignment.Center
                        HorizontalAlignment:=horizontal_alignment.Center
                        Padding:=margin{Left:=0.0, Top:=0.0, Right:=50.0, Bottom:=0.0}
                        #Acts as a frame background to format the header area
                        Widget:=texture_block{DefaultImage:=Textures.General.T_GenSquBlueTrans, DefaultDesiredSize:=vector2{X:=1800.0, Y:=200.0}, DefaultTint:=Green}

                    overlay_slot:
                        VerticalAlignment:=vertical_alignment.Center
                        HorizontalAlignment:=horizontal_alignment.Fill
                        Padding:=margin{Left:=0.0, Top:=0.0, Right:=0.0, Bottom:=0.0}
                        Widget:=stack_box:
                            Orientation:=orientation.Horizontal
                            Slots:=array:
                                # Header's horizontal Stack Box slot
                                stack_box_slot:
                                    VerticalAlignment:=vertical_alignment.Center
                                    HorizontalAlignment:=horizontal_alignment.Center
                                    Padding:=margin{Left:=100.0, Top:=0.0, Right:=200.0, Bottom:=0.0}
                                    #Empty space meant to keep the area for coin$ and diamonds
                                    Widget:=overlay:
                                        Slots:=array:
                                            #Makeshift size box for currency area
                                            overlay_slot:
                                                VerticalAlignment:=vertical_alignment.Center
                                                HorizontalAlignment:=horizontal_alignment.Center
                                                Padding:=margin{Left:=0.0, Top:=0.0, Right:=0.0, Bottom:=0.0}
                                                Widget:=texture_block{DefaultImage:=Textures.General.T_GenSquTrans, DefaultDesiredSize:=vector2{X:=490.0, Y:=60.0}, DefaultTint:=Red}#}DefaultTint:=Red

                                            #Horizontal stack box that will hold both currency frames
                                            overlay_slot:
                                                VerticalAlignment:=vertical_alignment.Top
                                                HorizontalAlignment:=horizontal_alignment.Left
                                                Padding:=margin{Left:=0.0, Top:=0.0, Right:=0.0, Bottom:=0.0}
                                                Widget:=stack_box:
                                                    Orientation:=orientation.Horizontal
                                                    Slots:=array:
                                                        stack_box_slot:
                                                            VerticalAlignment:=vertical_alignment.Center
                                                            HorizontalAlignment:=horizontal_alignment.Left
                                                            #Distribution:=option{0.5}
                                                            Padding:=margin{Left:=0.0, Top:=0.0, Right:=0.0, Bottom:=0.0}
                                                            Widget:=overlay:
                                                                Slots:=array:
                                                                    #Coin$ frame
                                                                    overlay_slot:
                                                                        VerticalAlignment:=vertical_alignment.Fill
                                                                        HorizontalAlignment:=horizontal_alignment.Left
                                                                        Padding:=margin{Left:=0.0, Top:=0.0, Right:=0.0, Bottom:=0.0}
                                                                        Widget:=texture_block{DefaultImage:=Textures.Currency.T_GoldSlot, DefaultDesiredSize:=vector2{X:=240.0, Y:=60.0}}
                                                                    #Coin$ value
                                                                    overlay_slot:
                                                                        VerticalAlignment:=vertical_alignment.Center
                                                                        HorizontalAlignment:=horizontal_alignment.Left
                                                                        Padding:=margin{Left:=0.0, Top:=0.0, Right:=0.0, Bottom:=0.0}
                                                                        Widget:=stack_box:
                                                                            # Horizontal stack box for the currency icon and value
                                                                            Orientation:=orientation.Horizontal
                                                                            Slots:=array:
                                                                                # Coin$ icon
                                                                                stack_box_slot:
                                                                                    VerticalAlignment:=vertical_alignment.Center
                                                                                    HorizontalAlignment:=horizontal_alignment.Center
                                                                                    Padding:=margin{Left:=5.0, Top:=0.0, Right:=0.0, Bottom:=0.0}
                                                                                    Widget:=texture_block{DefaultImage:=Textures.Currency.T_Gold, DefaultDesiredSize:=vector2{X:=50.0, Y:=50.0}}
                                                                                   
                                                                                # Coin$ value    
                                                                                stack_box_slot:
                                                                                    VerticalAlignment:=vertical_alignment.Center
                                                                                    HorizontalAlignment:=horizontal_alignment.Center
                                                                                    Padding:=margin{Left:=15.0, Top:=0.0, Right:=0.0, Bottom:=0.0}
                                                                                    Widget:=text_block{DefaultText:=StringToMessage("1000"), DefaultTextColor:=NamedColors.Black}

                                                        stack_box_slot:
                                                            VerticalAlignment:=vertical_alignment.Center
                                                            HorizontalAlignment:=horizontal_alignment.Right
                                                            Padding:=margin{Left:=10.0, Top:=0.0, Right:=0.0, Bottom:=0.0}
                                                            #Distribution:=option{0.5}
                                                            Widget:=overlay:
                                                                Slots:=array:
                                                                    #Diamonds Frame
                                                                    overlay_slot:
                                                                        VerticalAlignment:=vertical_alignment.Fill
                                                                        HorizontalAlignment:=horizontal_alignment.Left
                                                                        Padding:=margin{Left:=0.0, Top:=0.0, Right:=0.0, Bottom:=0.0}
                                                                        Widget:=texture_block{DefaultImage:=Textures.Currency.T_DiamondSlot, DefaultDesiredSize:=vector2{X:=240.0, Y:=60.0}}

                                                                    overlay_slot:
                                                                        VerticalAlignment:=vertical_alignment.Center
                                                                        HorizontalAlignment:=horizontal_alignment.Left
                                                                        Padding:=margin{Left:=0.0, Top:=0.0, Right:=0.0, Bottom:=0.0}
                                                                        Widget:=stack_box:
                                                                            # Horizontal stack box for the currency icon and value
                                                                            Orientation:=orientation.Horizontal
                                                                            Slots:=array:
                                                                                #dIOAMONDS icon
                                                                                stack_box_slot:
                                                                                    VerticalAlignment:=vertical_alignment.Center
                                                                                    HorizontalAlignment:=horizontal_alignment.Center
                                                                                    Padding:=margin{Left:=2.5, Top:=0.0, Right:=0.0, Bottom:=0.0}
                                                                                    Widget:=texture_block{DefaultImage:=Textures.Currency.T_Diamond, DefaultDesiredSize:=vector2{X:=50.0, Y:=50.0}}

                                                                                #Diamonds value
                                                                                stack_box_slot:
                                                                                    VerticalAlignment:=vertical_alignment.Center
                                                                                    HorizontalAlignment:=horizontal_alignment.Left
                                                                                    Padding:=margin{Left:=15.0, Top:=0.0, Right:=0.0, Bottom:=0.0}
                                                                                    Widget:=text_block{DefaultText:=StringToMessage("14"), DefaultTextColor:=NamedColors.Black}


                                stack_box_slot:
                                    VerticalAlignment:=vertical_alignment.Top
                                    HorizontalAlignment:=horizontal_alignment.Center
                                    Padding:=margin{Left:=105.0, Top:=5.0, Right:=330.0, Bottom:=0.0}
                                    #Header Text
                                    Widget:=TitleTextWidget

                                stack_box_slot:
                                    VerticalAlignment:=vertical_alignment.Center
                                    HorizontalAlignment:=horizontal_alignment.Right
                                    Padding:=margin{Left:=110.0, Top:=0.0, Right:=0.0, Bottom:=0.0}
                                    #Back Button
                                    Widget:=BackButton

            LoadoutCard:=overlay:
                        Slots:=array:
                            
                            #Makeshift size box for rarity flag consideration
                            overlay_slot:
                                VerticalAlignment:=vertical_alignment.Center
                                HorizontalAlignment:=horizontal_alignment.Center
                                #Distribution:=option{1.0}
                                Padding:=margin{Left:=0.0, Top:=0.0, Right:=0.0, Bottom:=0.0}
                                Widget:=texture_block{DefaultImage:=Textures.General.T_GenSquTrans, DefaultDesiredSize:=vector2{X:=486.0, Y:=685.0}, DefaultTint:=MakeColorFromHex("#FF0000FF")}
                                #LoadoutCard

                            #Overlay to centralize the content of the card
                            overlay_slot:
                                VerticalAlignment:=vertical_alignment.Bottom
                                HorizontalAlignment:=horizontal_alignment.Right
                                #Distribution:=option{1.0}
                                Padding:=margin{Left:=0.0, Top:=0.0, Right:=0.0, Bottom:=0.0}
                                Widget:=overlay:
                                    Slots:=array:

                                        #Background
                                        overlay_slot:
                                            VerticalAlignment:=vertical_alignment.Center
                                            HorizontalAlignment:=horizontal_alignment.Center
                                            #Distribution:=option{1.0}
                                            Padding:=margin{Left:=0.0, Top:=0.0, Right:=0.0, Bottom:=0.0}
                                            Widget:=texture_block{DefaultImage:=Textures.General.T_GenericSquare, DefaultDesiredSize:=vector2{X:=476.0, Y:=675.0}, DefaultTint:=MakeColorFromHex("#082958FF")}

                                        #Vertical slot for content
                                        overlay_slot:
                                            VerticalAlignment:=vertical_alignment.Center
                                            HorizontalAlignment:=horizontal_alignment.Center
                                            #Distribution:=option{1.0}
                                            Padding:=margin{Left:=0.0, Top:=0.0, Right:=0.0, Bottom:=0.0}
                                            Widget:=stack_box:
                                                Orientation:=orientation.Vertical
                                                Slots:=array:

                                                    #Card header
                                                    stack_box_slot:
                                                        VerticalAlignment:=vertical_alignment.Center
                                                        HorizontalAlignment:=horizontal_alignment.Center
                                                        #Distribution:=option{1.0}
                                                        Padding:=margin{Left:=0.0, Top:=0.0, Right:=0.0, Bottom:=15.0}
                                                        Widget:=text_block{DefaultText := StringToMessage("AVAILABLE WEAPONS"), DefaultTextColor:=MakeColorFromHex("#FFFFFFFF")}

                                                    #Loadout icons stack box
                                                    stack_box_slot:
                                                        VerticalAlignment:=vertical_alignment.Center
                                                        HorizontalAlignment:=horizontal_alignment.Center
                                                        #Distribution:=option{1.0}
                                                        Padding:=margin{Left:=0.0, Top:=0.0, Right:=0.0, Bottom:=0.0}
                                                        Widget:=stack_box:
                                                            Orientation:=orientation.Vertical
                                                            Slots:=array:

                                                                #First row for level weapons
                                                                stack_box_slot:
                                                                    VerticalAlignment:=vertical_alignment.Center
                                                                    HorizontalAlignment:=horizontal_alignment.Fill
                                                                    Distribution:=option{1.0}
                                                                    Padding:=margin{Left:=0.0, Top:=0.0, Right:=0.0, Bottom:=0.0}
                                                                    Widget:=stack_box:
                                                                        Orientation:=orientation.Horizontal
                                                                        Slots:=array:
                                                                
                                                                            #Pistol Icon
                                                                            stack_box_slot:
                                                                                VerticalAlignment:=vertical_alignment.Center
                                                                                HorizontalAlignment:=horizontal_alignment.Center
                                                                                Padding:=margin{Left:=0.0, Top:=0.0, Right:=8.0, Bottom:=0.0}
                                                                                Widget:=overlay:

                                                                                    Slots:=array:
                                                                                        #Weapon Icon - Pistols
                                                                                        overlay_slot:
                                                                                            VerticalAlignment:=vertical_alignment.Center
                                                                                            HorizontalAlignment:=horizontal_alignment.Center
                                                                                            Padding:=margin{Left:=0.0, Top:=0.0, Right:=0.0, Bottom:=0.0}
                                                                                            #Weapon icon BG
                                                                                            Widget:=texture_block{DefaultImage:=Textures.General.T_GenericSquare, DefaultDesiredSize:=vector2{X:=100.0, Y:=100.0}, DefaultTint:=MakeColorFromHex("#000c1d88")}
                                            
                                                                                        overlay_slot:
                                                                                            VerticalAlignment:=vertical_alignment.Center
                                                                                            HorizontalAlignment:=horizontal_alignment.Center
                                                                                            Padding:=margin{Left:=0.0, Top:=0.0, Right:=0.0, Bottom:=0.0}
                                                                                            #Weapon icon Border
                                                                                            Widget:=texture_block{DefaultImage:=Textures.General.T_GenericBorderSquare, DefaultDesiredSize:=vector2{X:=100.0, Y:=100.0}, DefaultTint:=MakeColorFromHex("#2e7be3FF")}
                                            
                                                                                        overlay_slot:
                                                                                            VerticalAlignment:=vertical_alignment.Center
                                                                                            HorizontalAlignment:=horizontal_alignment.Center
                                                                                            Padding:=margin{Left:=0.0, Top:=0.0, Right:=0.0, Bottom:=0.0}
                                                                                            #Weapon icon - pistol
                                                                                            Widget:=texture_block{DefaultImage:=Textures.Items.T_IconPistol, DefaultDesiredSize:=vector2{X:=85.0, Y:=85.0}, DefaultTint:=MakeColorFromHex("#FFFFFFFF")}

                                                                            #Shotgun Icon
                                                                            stack_box_slot:
                                                                                VerticalAlignment:=vertical_alignment.Center
                                                                                HorizontalAlignment:=horizontal_alignment.Center
                                                                                Padding:=margin{Left:=0.0, Top:=0.0, Right:=0.0, Bottom:=0.0}
                                                                                Widget:=overlay:

                                                                                    Slots:=array:
                                                                                        #Weapon Icon - Pistols
                                                                                        overlay_slot:
                                                                                            VerticalAlignment:=vertical_alignment.Center
                                                                                            HorizontalAlignment:=horizontal_alignment.Center
                                                                                            Padding:=margin{Left:=0.0, Top:=0.0, Right:=0.0, Bottom:=0.0}
                                                                                            #Weapon icon BG
                                                                                            Widget:=texture_block{DefaultImage:=Textures.General.T_GenericSquare, DefaultDesiredSize:=vector2{X:=100.0, Y:=100.0}, DefaultTint:=MakeColorFromHex("#000c1d88")}

                                                                                        overlay_slot:
                                                                                            VerticalAlignment:=vertical_alignment.Center
                                                                                            HorizontalAlignment:=horizontal_alignment.Center
                                                                                            Padding:=margin{Left:=0.0, Top:=0.0, Right:=0.0, Bottom:=0.0}
                                                                                            #Weapon icon Border
                                                                                            Widget:=texture_block{DefaultImage:=Textures.General.T_GenericBorderSquare, DefaultDesiredSize:=vector2{X:=100.0, Y:=100.0}, DefaultTint:=MakeColorFromHex("#2e7be3FF")}
                                            
                                                                                        overlay_slot:
                                                                                            VerticalAlignment:=vertical_alignment.Center
                                                                                            HorizontalAlignment:=horizontal_alignment.Center
                                                                                            Padding:=margin{Left:=0.0, Top:=0.0, Right:=0.0, Bottom:=0.0}
                                                                                            #Weapon icon - pistol
                                                                                            Widget:=texture_block{DefaultImage:=Textures.Items.T_IconShotgun, DefaultDesiredSize:=vector2{X:=85.0, Y:=85.0}, DefaultTint:=MakeColorFromHex("#FFFFFFFF")}

                                                                #Second row for level weapons
                                                                stack_box_slot:
                                                                    VerticalAlignment:=vertical_alignment.Center
                                                                    HorizontalAlignment:=horizontal_alignment.Fill
                                                                    Distribution:=option{1.0}
                                                                    Padding:=margin{Left:=0.0, Top:=8.0, Right:=8.0, Bottom:=0.0}
                                                                    Widget:=stack_box:
                                                                        Orientation:=orientation.Horizontal
                                                                        Slots:=array:                                                                            

                                                                            #SMG Icon
                                                                            stack_box_slot:
                                                                                VerticalAlignment:=vertical_alignment.Center
                                                                                HorizontalAlignment:=horizontal_alignment.Center
                                                                                Padding:=margin{Left:=0.0, Top:=0.0, Right:=8.0, Bottom:=0.0}
                                                                                Widget:=overlay:

                                                                                    Slots:=array:
                                                                                        #Weapon Icon - Pistols
                                                                                        overlay_slot:
                                                                                            VerticalAlignment:=vertical_alignment.Center
                                                                                            HorizontalAlignment:=horizontal_alignment.Center
                                                                                            Padding:=margin{Left:=0.0, Top:=0.0, Right:=0.0, Bottom:=0.0}
                                                                                            #Weapon icon BG
                                                                                            Widget:=texture_block{DefaultImage:=Textures.General.T_GenericSquare, DefaultDesiredSize:=vector2{X:=100.0, Y:=100.0}, DefaultTint:=MakeColorFromHex("#000c1d88")}
                                            
                                                                                        overlay_slot:
                                                                                            VerticalAlignment:=vertical_alignment.Center
                                                                                            HorizontalAlignment:=horizontal_alignment.Center
                                                                                            Padding:=margin{Left:=0.0, Top:=0.0, Right:=0.0, Bottom:=0.0}
                                                                                            #Weapon icon Border
                                                                                            Widget:=texture_block{DefaultImage:=Textures.General.T_GenericBorderSquare, DefaultDesiredSize:=vector2{X:=100.0, Y:=100.0}, DefaultTint:=MakeColorFromHex("#2e7be3FF")}
                                            
                                                                                        overlay_slot:
                                                                                            VerticalAlignment:=vertical_alignment.Center
                                                                                            HorizontalAlignment:=horizontal_alignment.Center
                                                                                            Padding:=margin{Left:=0.0, Top:=0.0, Right:=0.0, Bottom:=0.0}
                                                                                            #Weapon icon - pistol
                                                                                            Widget:=texture_block{DefaultImage:=Textures.Items.T_IconSMG, DefaultDesiredSize:=vector2{X:=85.0, Y:=85.0}, DefaultTint:=MakeColorFromHex("#FFFFFFFF")}

                                                                            #Rifle Icon
                                                                            stack_box_slot:
                                                                                VerticalAlignment:=vertical_alignment.Center
                                                                                HorizontalAlignment:=horizontal_alignment.Center
                                                                                Padding:=margin{Left:=0.0, Top:=0.0, Right:=8.0, Bottom:=0.0}
                                                                                Widget:=overlay:

                                                                                    Slots:=array:
                                                                                        #Weapon Icon - Pistols
                                                                                        overlay_slot:
                                                                                            VerticalAlignment:=vertical_alignment.Center
                                                                                            HorizontalAlignment:=horizontal_alignment.Center
                                                                                            Padding:=margin{Left:=0.0, Top:=0.0, Right:=0.0, Bottom:=0.0}
                                                                                            #Weapon icon BG
                                                                                            Widget:=texture_block{DefaultImage:=Textures.General.T_GenericSquare, DefaultDesiredSize:=vector2{X:=100.0, Y:=100.0}, DefaultTint:=MakeColorFromHex("#000c1d88")}
                                            
                                                                                        overlay_slot:
                                                                                            VerticalAlignment:=vertical_alignment.Center
                                                                                            HorizontalAlignment:=horizontal_alignment.Center
                                                                                            Padding:=margin{Left:=0.0, Top:=0.0, Right:=0.0, Bottom:=0.0}
                                                                                            #Weapon icon Border
                                                                                            Widget:=texture_block{DefaultImage:=Textures.General.T_GenericBorderSquare, DefaultDesiredSize:=vector2{X:=100.0, Y:=100.0}, DefaultTint:=MakeColorFromHex("#2e7be3FF")}
                                            
                                                                                        overlay_slot:
                                                                                            VerticalAlignment:=vertical_alignment.Center
                                                                                            HorizontalAlignment:=horizontal_alignment.Center
                                                                                            Padding:=margin{Left:=0.0, Top:=0.0, Right:=0.0, Bottom:=0.0}
                                                                                            #Weapon icon - pistol
                                                                                            Widget:=texture_block{DefaultImage:=Textures.Items.T_IconRifle, DefaultDesiredSize:=vector2{X:=85.0, Y:=85.0}, DefaultTint:=MakeColorFromHex("#FFFFFFFF")}

                                                    #Loadout text digest
                                                    stack_box_slot:
                                                        VerticalAlignment:=vertical_alignment.Center
                                                        HorizontalAlignment:=horizontal_alignment.Center
                                                        Padding:=margin{Left:=0.0, Top:=0.0, Right:=0.0, Bottom:=0.0}
                                                        Widget:=overlay:
                                                            Slots:=array:

                                                                #Makeshift size box for text
                                                                overlay_slot:
                                                                    VerticalAlignment:=vertical_alignment.Center
                                                                    HorizontalAlignment:=horizontal_alignment.Center
                                                                    Padding:=margin{Left:=0.0, Top:=0.0, Right:=0.0, Bottom:=0.0}
                                                                    Widget:=texture_block{DefaultImage:=Textures.General.T_GenSquTrans, DefaultDesiredSize:=vector2{X:=170.0, Y:=175.0}, DefaultTint:=MakeColorFromHex("#FF0000FF")}

                                                                #Vertical Text digest
                                                                overlay_slot:
                                                                    VerticalAlignment:=vertical_alignment.Center
                                                                    HorizontalAlignment:=horizontal_alignment.Left
                                                                    Padding:=margin{Left:=0.0, Top:=0.0, Right:=0.0, Bottom:=0.0}
                                                                    Widget:=stack_box:

                                                                        Orientation:=orientation.Vertical
                                                                        Slots:=array:

                                                                            stack_box_slot:
                                                                                VerticalAlignment:=vertical_alignment.Center
                                                                                HorizontalAlignment:=horizontal_alignment.Left
                                                                                Padding:=margin{Left:=0.0, Top:=0.0, Right:=0.0, Bottom:=0.0}
                                                                                Widget:=text_block{DefaultText := StringToMessage("• Pistols"), DefaultTextColor:=MakeColorFromHex("#FFFFFFFF")}

                                                                            stack_box_slot:
                                                                                VerticalAlignment:=vertical_alignment.Center
                                                                                HorizontalAlignment:=horizontal_alignment.Left
                                                                                Padding:=margin{Left:=0.0, Top:=0.0, Right:=0.0, Bottom:=0.0}
                                                                                Widget:=text_block{DefaultText := StringToMessage("• Shotguns"), DefaultTextColor:=MakeColorFromHex("#FFFFFFFF")}

                                                                            stack_box_slot:
                                                                                VerticalAlignment:=vertical_alignment.Center
                                                                                HorizontalAlignment:=horizontal_alignment.Left
                                                                                Padding:=margin{Left:=0.0, Top:=0.0, Right:=0.0, Bottom:=0.0}
                                                                                Widget:=text_block{DefaultText := StringToMessage("• SMGs"), DefaultTextColor:=MakeColorFromHex("#FFFFFFFF")}

                                                                            stack_box_slot:
                                                                                VerticalAlignment:=vertical_alignment.Center
                                                                                HorizontalAlignment:=horizontal_alignment.Left
                                                                                Padding:=margin{Left:=0.0, Top:=0.0, Right:=0.0, Bottom:=0.0}
                                                                                Widget:=text_block{DefaultText := StringToMessage("• Rifles"), DefaultTextColor:=MakeColorFromHex("#FFFFFFFF")}

                                                    stack_box_slot:
                                                        VerticalAlignment:=vertical_alignment.Center
                                                        HorizontalAlignment:=horizontal_alignment.Center
                                                        Padding:=margin{Left:=0.0, Top:=0.0, Right:=0.0, Bottom:=0.0}
                                                        Widget:=button_loud{DefaultText:=StringToMessage("UNLOCK LEVEL 2\n$25")}

                            #Level flag
                            overlay_slot:
                                VerticalAlignment:=vertical_alignment.Top
                                HorizontalAlignment:=horizontal_alignment.Left
                                Padding:=margin{Left:=0.0, Top:=0.0, Right:=15.0, Bottom:=0.0}
                                Widget:=overlay:

                                    Slots:=array:
                                        overlay_slot:
                                            VerticalAlignment:=vertical_alignment.Center
                                            HorizontalAlignment:=horizontal_alignment.Center
                                            Padding:=margin{Left:=0.0, Top:=0.0, Right:=0.0, Bottom:=0.0}
                                            #Tier level makeshift size box
                                            Widget:=texture_block{DefaultImage:=Textures.General.T_GenSquTrans, DefaultDesiredSize:=vector2{X:=150.0, Y:=114.0}, DefaultTint:=MakeColorFromHex("#FF0000FF")}

                                        overlay_slot:
                                            VerticalAlignment:=vertical_alignment.Top
                                            HorizontalAlignment:=horizontal_alignment.Left
                                            Padding:=margin{Left:=0.0, Top:=0.0, Right:=0.0, Bottom:=0.0}
                                            #Selected level yellow flag
                                            Widget:=overlay:

                                                Slots:=array:
                                                    overlay_slot:
                                                        VerticalAlignment:=vertical_alignment.Center
                                                        HorizontalAlignment:=horizontal_alignment.Center
                                                        Padding:=margin{Left:=0.0, Top:=0.0, Right:=0.0, Bottom:=0.0}
                                                        #level flag texture
                                                        Widget:=texture_block{DefaultImage:=Textures.General.T_GenLevelFlag, DefaultDesiredSize:=vector2{X:=140.0, Y:=45.0}, DefaultTint:=MakeColorFromHex("#f7ff1aFF")}

                                                    overlay_slot:
                                                        VerticalAlignment:=vertical_alignment.Center
                                                        HorizontalAlignment:=horizontal_alignment.Center
                                                        Padding:=margin{Left:=0.0, Top:=0.0, Right:=0.0, Bottom:=0.0}
                                                        #level flag text
                                                        Widget:=text_block{DefaultText := StringToMessage("Level 1"), DefaultTextColor:=MakeColorFromHex("#000000FF")}
            
            #Definition for the default level rows
            LevelRow0:=overlay:
                Slots:=array:
                    overlay_slot:
                        #Makeshift size box
                        VerticalAlignment:=vertical_alignment.Center
                        HorizontalAlignment:=horizontal_alignment.Center
                        Padding:=margin{Left:=0.0, Top:=0.0, Right:=0.0, Bottom:=0.0}
                        Widget:=texture_block{DefaultImage:=Textures.General.T_GenSquTrans, DefaultDesiredSize:=vector2{X:=1250.0, Y:=114.0}, DefaultTint:=MakeColorFromHex("#0000FFFF")}

                    overlay_slot:
                        #Rarity level content
                        VerticalAlignment:=vertical_alignment.Center
                        HorizontalAlignment:=horizontal_alignment.Left
                        Padding:=margin{Left:=0.0, Top:=0.0, Right:=0.0, Bottom:=0.0}
                        Widget:=overlay:
                            Slots:=array:

                                overlay_slot:
                                    #Makeshift size box
                                    VerticalAlignment:=vertical_alignment.Center
                                    HorizontalAlignment:=horizontal_alignment.Fill
                                    Padding:=margin{Left:=0.0, Top:=0.0, Right:=0.0, Bottom:=0.0}
                                    Widget:=texture_block{DefaultImage:=Textures.General.T_GenericSquare, DefaultDesiredSize:=vector2{X:=1190.0, Y:=114.0}, DefaultTint:=MakeColorFromHex("#19373e32")}

                                overlay_slot:
                                    #Highlight texture, over the content but under the rarity flag
                                    VerticalAlignment:=vertical_alignment.Fill
                                    HorizontalAlignment:=horizontal_alignment.Fill
                                    Padding:=margin{Left:=0.0, Top:=0.0, Right:=0.0, Bottom:=0.0}
                                    Widget:=texture_block{DefaultImage:=Textures.General.T_LevelBorder, DefaultDesiredSize:=vector2{X:=1190.0, Y:=64.0}, DefaultTint:=MakeColorFromHex("#1e5f4cFF")}

                                    #Horizontal stack box
                                overlay_slot:
                                    VerticalAlignment:=vertical_alignment.Fill
                                    HorizontalAlignment:=horizontal_alignment.Left
                                    Padding:=margin{Left:=0.0, Top:=0.0, Right:=0.0, Bottom:=0.0}
                                    Widget:=stack_box:
                                        Orientation:=orientation.Horizontal
                                        Slots:=array:
                                            #Level flag
                                            stack_box_slot:
                                                VerticalAlignment:=vertical_alignment.Top
                                                HorizontalAlignment:=horizontal_alignment.Left
                                                Padding:=margin{Left:=0.0, Top:=0.0, Right:=15.0, Bottom:=0.0}
                                                Widget:=overlay:

                                                    Slots:=array:
                                                        overlay_slot:
                                                            VerticalAlignment:=vertical_alignment.Center
                                                            HorizontalAlignment:=horizontal_alignment.Center
                                                            Padding:=margin{Left:=0.0, Top:=0.0, Right:=0.0, Bottom:=0.0}
                                                            #Tier level makeshift size box
                                                            Widget:=texture_block{DefaultImage:=Textures.General.T_GenSquTrans, DefaultDesiredSize:=vector2{X:=150.0, Y:=114.0}, DefaultTint:=MakeColorFromHex("#FF0000FF")}

                                                        #overlay_slot:
                                                            #VerticalAlignment:=vertical_alignment.Top
                                                            #HorizontalAlignment:=horizontal_alignment.Left
                                                            #Padding:=margin{Left:=0.0, Top:=0.0, Right:=0.0, Bottom:=0.0}
                                                            #Selected level yellow flag
                                                            #Widget:=overlay:

                                                                #Slots:=array:
                                                                    #overlay_slot:
                                                                        #VerticalAlignment:=vertical_alignment.Center
                                                                        #HorizontalAlignment:=horizontal_alignment.Center
                                                                        #Padding:=margin{Left:=0.0, Top:=0.0, Right:=0.0, Bottom:=0.0}
                                                                        #level flag texture
                                                                        #Widget:=texture_block{DefaultImage:=Textures.General.T_GenLevelFlag, DefaultDesiredSize:=vector2{X:=140.0, Y:=45.0}, DefaultTint:=MakeColorFromHex("#f7ff1aFF")}

                                                                    #overlay_slot:
                                                                        #VerticalAlignment:=vertical_alignment.Center
                                                                        #HorizontalAlignment:=horizontal_alignment.Center
                                                                        #Padding:=margin{Left:=0.0, Top:=0.0, Right:=0.0, Bottom:=0.0}
                                                                        #level flag text
                                                                        #Widget:=text_block{DefaultText := StringToMessage("Level 0"), DefaultTextColor:=MakeColorFromHex("#000000FF")}

                                                        overlay_slot:
                                                            VerticalAlignment:=vertical_alignment.Fill
                                                            HorizontalAlignment:=horizontal_alignment.Center
                                                            Padding:=margin{Left:=0.0, Top:=5.0, Right:=0.0, Bottom:=5.0}
                                                            #Unselected level and lock
                                                            Widget:=overlay:
                                                                Slots:=array:
                                                                    
                                                                    overlay_slot:
                                                                        VerticalAlignment:=vertical_alignment.Top
                                                                        HorizontalAlignment:=horizontal_alignment.Center
                                                                        Padding:=margin{Left:=0.0, Top:=0.0, Right:=0.0, Bottom:=0.0}
                                                                        #level flag texture
                                                                        Widget:=text_block{DefaultText := StringToMessage("Level 0"), DefaultTextColor:=MakeColorFromHex("#FFFFFFFF")}

                                                                    #overlay_slot:
                                                                        #VerticalAlignment:=vertical_alignment.Bottom
                                                                        #HorizontalAlignment:=horizontal_alignment.Center
                                                                        #Padding:=margin{Left:=0.0, Top:=0.0, Right:=0.0, Bottom:=0.0}
                                                                        #level flag texture
                                                                        #Widget:=texture_block{DefaultImage:=Textures.General.T_Lock, DefaultDesiredSize:=vector2{X:=52.0, Y:=58.0}, DefaultTint:=MakeColorFromHex("#ffffff88")}
                                            
                                            #Level weapon roster
                                            stack_box_slot:
                                                VerticalAlignment:=vertical_alignment.Center
                                                HorizontalAlignment:=horizontal_alignment.Center
                                                Padding:=margin{Left:=0.0, Top:=0.0, Right:=8.0, Bottom:=0.0}
                                                Widget:=overlay:
                                                    Slots:=array:
                                                        #Weapon Icon - Pistols
                                                        overlay_slot:
                                                            VerticalAlignment:=vertical_alignment.Center
                                                            HorizontalAlignment:=horizontal_alignment.Center
                                                            Padding:=margin{Left:=0.0, Top:=0.0, Right:=0.0, Bottom:=0.0}
                                                            #Weapon icon BG
                                                            Widget:=texture_block{DefaultImage:=Textures.General.T_GenericSquare, DefaultDesiredSize:=vector2{X:=78.0, Y:=78.0}, DefaultTint:=MakeColorFromHex("#000c1d88")}

                                                        overlay_slot:
                                                            VerticalAlignment:=vertical_alignment.Center
                                                            HorizontalAlignment:=horizontal_alignment.Center
                                                            Padding:=margin{Left:=0.0, Top:=0.0, Right:=0.0, Bottom:=0.0}
                                                            #Weapon icon Border
                                                            Widget:=texture_block{DefaultImage:=Textures.General.T_GenericBorderSquare, DefaultDesiredSize:=vector2{X:=78.0, Y:=78.0}, DefaultTint:=MakeColorFromHex("#2e7be3FF")}

                                                        overlay_slot:
                                                            VerticalAlignment:=vertical_alignment.Top
                                                            HorizontalAlignment:=horizontal_alignment.Center
                                                            Padding:=margin{Left:=0.0, Top:=5.0, Right:=0.0, Bottom:=0.0}
                                                            #Weapon icon - pistol
                                                            Widget:=texture_block{DefaultImage:=Textures.Items.T_IconPistol, DefaultDesiredSize:=vector2{X:=50.0, Y:=50.0}, DefaultTint:=MakeColorFromHex("#FFFFFFFF")}

                                                        overlay_slot:
                                                            VerticalAlignment:=vertical_alignment.Bottom
                                                            HorizontalAlignment:=horizontal_alignment.Center
                                                            Padding:=margin{Left:=0.0, Top:=0.0, Right:=0.0, Bottom:=5.0}
                                                            #Weapon header - pistol
                                                            Widget:=texture_block{DefaultImage:=Textures.Items.T_HeaderPistol, DefaultDesiredSize:=vector2{X:=78.0, Y:=18.0}, DefaultTint:=MakeColorFromHex("#FFFFFFFF")}
                                                        
                                            stack_box_slot:
                                                VerticalAlignment:=vertical_alignment.Center
                                                HorizontalAlignment:=horizontal_alignment.Center
                                                Padding:=margin{Left:=0.0, Top:=0.0, Right:=8.0, Bottom:=0.0}
                                                Widget:=overlay:
                                                    Slots:=array:
                                                        #Weapon Icon - Shotgun
                                                        overlay_slot:
                                                            VerticalAlignment:=vertical_alignment.Center
                                                            HorizontalAlignment:=horizontal_alignment.Center
                                                            Padding:=margin{Left:=0.0, Top:=0.0, Right:=0.0, Bottom:=0.0}
                                                            #Weapon icon BG
                                                            Widget:=texture_block{DefaultImage:=Textures.General.T_GenericSquare, DefaultDesiredSize:=vector2{X:=78.0, Y:=78.0}, DefaultTint:=MakeColorFromHex("#000c1d88")}

                                                        overlay_slot:
                                                            VerticalAlignment:=vertical_alignment.Center
                                                            HorizontalAlignment:=horizontal_alignment.Center
                                                            Padding:=margin{Left:=0.0, Top:=0.0, Right:=0.0, Bottom:=0.0}
                                                            #Weapon icon Border
                                                            Widget:=texture_block{DefaultImage:=Textures.General.T_GenericBorderSquare, DefaultDesiredSize:=vector2{X:=78.0, Y:=78.0}, DefaultTint:=MakeColorFromHex("#2e7be3FF")}

                                                        overlay_slot:
                                                            VerticalAlignment:=vertical_alignment.Top
                                                            HorizontalAlignment:=horizontal_alignment.Center
                                                            Padding:=margin{Left:=0.0, Top:=5.0, Right:=0.0, Bottom:=0.0}
                                                            #Weapon icon - shotgun
                                                            Widget:=texture_block{DefaultImage:=Textures.Items.T_IconShotgun, DefaultDesiredSize:=vector2{X:=50.0, Y:=50.0}, DefaultTint:=MakeColorFromHex("#FFFFFFFF")}

                                                        overlay_slot:
                                                            VerticalAlignment:=vertical_alignment.Bottom
                                                            HorizontalAlignment:=horizontal_alignment.Center
                                                            Padding:=margin{Left:=0.0, Top:=0.0, Right:=0.0, Bottom:=5.0}
                                                            #Weapon header - shotgun
                                                            Widget:=texture_block{DefaultImage:=Textures.Items.T_HeaderShotgun, DefaultDesiredSize:=vector2{X:=78.0, Y:=18.0}, DefaultTint:=MakeColorFromHex("#FFFFFFFF")}

                                            stack_box_slot:
                                                VerticalAlignment:=vertical_alignment.Center
                                                HorizontalAlignment:=horizontal_alignment.Center
                                                Padding:=margin{Left:=0.0, Top:=0.0, Right:=8.0, Bottom:=0.0}
                                                Widget:=overlay:
                                                    Slots:=array:
                                                        #Weapon Icon - SMG
                                                        overlay_slot:
                                                            VerticalAlignment:=vertical_alignment.Center
                                                            HorizontalAlignment:=horizontal_alignment.Center
                                                            Padding:=margin{Left:=0.0, Top:=0.0, Right:=0.0, Bottom:=0.0}
                                                            #Weapon icon BG
                                                            Widget:=texture_block{DefaultImage:=Textures.General.T_GenericSquare, DefaultDesiredSize:=vector2{X:=78.0, Y:=78.0}, DefaultTint:=MakeColorFromHex("#000c1d88")}

                                                        overlay_slot:
                                                            VerticalAlignment:=vertical_alignment.Center
                                                            HorizontalAlignment:=horizontal_alignment.Center
                                                            Padding:=margin{Left:=0.0, Top:=0.0, Right:=0.0, Bottom:=0.0}
                                                            #Weapon icon Border
                                                            Widget:=texture_block{DefaultImage:=Textures.General.T_GenericBorderSquare, DefaultDesiredSize:=vector2{X:=78.0, Y:=78.0}, DefaultTint:=MakeColorFromHex("#2e7be3FF")}

                                                        overlay_slot:
                                                            VerticalAlignment:=vertical_alignment.Top
                                                            HorizontalAlignment:=horizontal_alignment.Center
                                                            Padding:=margin{Left:=0.0, Top:=5.0, Right:=0.0, Bottom:=0.0}
                                                            #Weapon icon - SMG
                                                            Widget:=texture_block{DefaultImage:=Textures.Items.T_IconSMG, DefaultDesiredSize:=vector2{X:=50.0, Y:=50.0}, DefaultTint:=MakeColorFromHex("#FFFFFFFF")}

                                                        overlay_slot:
                                                            VerticalAlignment:=vertical_alignment.Bottom
                                                            HorizontalAlignment:=horizontal_alignment.Center
                                                            Padding:=margin{Left:=0.0, Top:=0.0, Right:=0.0, Bottom:=5.0}
                                                            #Weapon header - SMG
                                                            Widget:=texture_block{DefaultImage:=Textures.Items.T_HeaderSMGs, DefaultDesiredSize:=vector2{X:=78.0, Y:=18.0}, DefaultTint:=MakeColorFromHex("#FFFFFFFF")}
                                                            

                                            stack_box_slot:
                                                VerticalAlignment:=vertical_alignment.Center
                                                HorizontalAlignment:=horizontal_alignment.Center
                                                Padding:=margin{Left:=0.0, Top:=0.0, Right:=8.0, Bottom:=0.0}
                                                Widget:=overlay:
                                                    Slots:=array:
                                                        #Weapon Icon - Rifle
                                                        overlay_slot:
                                                            VerticalAlignment:=vertical_alignment.Center
                                                            HorizontalAlignment:=horizontal_alignment.Center
                                                            Padding:=margin{Left:=0.0, Top:=0.0, Right:=0.0, Bottom:=0.0}
                                                            #Weapon icon BG
                                                            Widget:=texture_block{DefaultImage:=Textures.General.T_GenericSquare, DefaultDesiredSize:=vector2{X:=78.0, Y:=78.0}, DefaultTint:=MakeColorFromHex("#000c1d88")}

                                                        overlay_slot:
                                                            VerticalAlignment:=vertical_alignment.Center
                                                            HorizontalAlignment:=horizontal_alignment.Center
                                                            Padding:=margin{Left:=0.0, Top:=0.0, Right:=0.0, Bottom:=0.0}
                                                            #Weapon icon Border
                                                            Widget:=texture_block{DefaultImage:=Textures.General.T_GenericBorderSquare, DefaultDesiredSize:=vector2{X:=78.0, Y:=78.0}, DefaultTint:=MakeColorFromHex("#2e7be3FF")}

                                                        overlay_slot:
                                                            VerticalAlignment:=vertical_alignment.Top
                                                            HorizontalAlignment:=horizontal_alignment.Center
                                                            Padding:=margin{Left:=0.0, Top:=5.0, Right:=0.0, Bottom:=0.0}
                                                            #Weapon icon - Rifle
                                                            Widget:=texture_block{DefaultImage:=Textures.Items.T_IconRifle, DefaultDesiredSize:=vector2{X:=50.0, Y:=50.0}, DefaultTint:=MakeColorFromHex("#FFFFFFFF")}

                                                        overlay_slot:
                                                            VerticalAlignment:=vertical_alignment.Bottom
                                                            HorizontalAlignment:=horizontal_alignment.Center
                                                            Padding:=margin{Left:=0.0, Top:=0.0, Right:=0.0, Bottom:=5.0}
                                                            #Weapon header - Rifle
                                                            Widget:=texture_block{DefaultImage:=Textures.Items.T_HeaderRifle, DefaultDesiredSize:=vector2{X:=78.0, Y:=18.0}, DefaultTint:=MakeColorFromHex("#FFFFFFFF")}

                                            stack_box_slot:
                                                VerticalAlignment:=vertical_alignment.Center
                                                HorizontalAlignment:=horizontal_alignment.Center
                                                Padding:=margin{Left:=0.0, Top:=0.0, Right:=8.0, Bottom:=0.0}
                                                Widget:=overlay:
                                                    Slots:=array:
                                                        #Weapon Icon - Explosive
                                                        overlay_slot:
                                                            VerticalAlignment:=vertical_alignment.Center
                                                            HorizontalAlignment:=horizontal_alignment.Center
                                                            Padding:=margin{Left:=0.0, Top:=0.0, Right:=0.0, Bottom:=0.0}
                                                            #Weapon icon BG
                                                            Widget:=texture_block{DefaultImage:=Textures.General.T_GenericSquare, DefaultDesiredSize:=vector2{X:=78.0, Y:=78.0}, DefaultTint:=MakeColorFromHex("#000c1d88")}

                                                        overlay_slot:
                                                            VerticalAlignment:=vertical_alignment.Center
                                                            HorizontalAlignment:=horizontal_alignment.Center
                                                            Padding:=margin{Left:=0.0, Top:=0.0, Right:=0.0, Bottom:=0.0}
                                                            #Weapon icon Border
                                                            Widget:=texture_block{DefaultImage:=Textures.General.T_GenericBorderSquare, DefaultDesiredSize:=vector2{X:=78.0, Y:=78.0}, DefaultTint:=MakeColorFromHex("#2e7be3FF")}

                                                        #overlay_slot:
                                                            #VerticalAlignment:=vertical_alignment.Top
                                                            #HorizontalAlignment:=horizontal_alignment.Center
                                                            #Padding:=margin{Left:=0.0, Top:=5.0, Right:=0.0, Bottom:=0.0}
                                                            #Weapon icon - Explosive
                                                            #Widget:=texture_block{DefaultImage:=Textures.Items.T_IconExplosive, DefaultDesiredSize:=vector2{X:=50.0, Y:=50.0}, DefaultTint:=MakeColorFromHex("#FFFFFFFF")}

                                                        #overlay_slot:
                                                            #VerticalAlignment:=vertical_alignment.Bottom
                                                            #HorizontalAlignment:=horizontal_alignment.Center
                                                            #Padding:=margin{Left:=0.0, Top:=0.0, Right:=0.0, Bottom:=5.0}
                                                            #Weapon header - Explosive
                                                            #Widget:=texture_block{DefaultImage:=Textures.Items.T_HeaderExplosive, DefaultDesiredSize:=vector2{X:=78.0, Y:=18.0}, DefaultTint:=MakeColorFromHex("#FFFFFFFF")}

                                            stack_box_slot:
                                                VerticalAlignment:=vertical_alignment.Center
                                                HorizontalAlignment:=horizontal_alignment.Center
                                                Padding:=margin{Left:=0.0, Top:=0.0, Right:=8.0, Bottom:=0.0}
                                                Widget:=overlay:
                                                    Slots:=array:
                                                        #Weapon Icon - Special
                                                        overlay_slot:
                                                            VerticalAlignment:=vertical_alignment.Center
                                                            HorizontalAlignment:=horizontal_alignment.Center
                                                            Padding:=margin{Left:=0.0, Top:=0.0, Right:=0.0, Bottom:=0.0}
                                                            #Weapon icon BG
                                                            Widget:=texture_block{DefaultImage:=Textures.General.T_GenericSquare, DefaultDesiredSize:=vector2{X:=78.0, Y:=78.0}, DefaultTint:=MakeColorFromHex("#000c1d88")}

                                                        overlay_slot:
                                                            VerticalAlignment:=vertical_alignment.Center
                                                            HorizontalAlignment:=horizontal_alignment.Center
                                                            Padding:=margin{Left:=0.0, Top:=0.0, Right:=0.0, Bottom:=0.0}
                                                            #Weapon icon Border
                                                            Widget:=texture_block{DefaultImage:=Textures.General.T_GenericBorderSquare, DefaultDesiredSize:=vector2{X:=78.0, Y:=78.0}, DefaultTint:=MakeColorFromHex("#2e7be3FF")}

                                                        #overlay_slot:
                                                            #VerticalAlignment:=vertical_alignment.Top
                                                            #HorizontalAlignment:=horizontal_alignment.Center
                                                            #Padding:=margin{Left:=0.0, Top:=5.0, Right:=0.0, Bottom:=0.0}
                                                            #Weapon icon - special
                                                            #Widget:=texture_block{DefaultImage:=Textures.Items.T_IconArtifact, DefaultDesiredSize:=vector2{X:=50.0, Y:=50.0}, DefaultTint:=MakeColorFromHex("#FFFFFFFF")}

                                                        #overlay_slot:
                                                            #VerticalAlignment:=vertical_alignment.Bottom
                                                            #HorizontalAlignment:=horizontal_alignment.Center
                                                            #Padding:=margin{Left:=0.0, Top:=0.0, Right:=0.0, Bottom:=5.0}
                                                            #Weapon header - Special
                                                            #Widget:=texture_block{DefaultImage:=Textures.Items.T_HeaderSpecial, DefaultDesiredSize:=vector2{X:=78.0, Y:=18.0}, DefaultTint:=MakeColorFromHex("#FFFFFFFF")}

                                            stack_box_slot:
                                                VerticalAlignment:=vertical_alignment.Center
                                                HorizontalAlignment:=horizontal_alignment.Center
                                                Padding:=margin{Left:=0.0, Top:=0.0, Right:=8.0, Bottom:=0.0}
                                                Widget:=overlay:
                                                    Slots:=array:
                                                        #Weapon Icon - DMR Sniper
                                                        overlay_slot:
                                                            VerticalAlignment:=vertical_alignment.Center
                                                            HorizontalAlignment:=horizontal_alignment.Center
                                                            Padding:=margin{Left:=0.0, Top:=0.0, Right:=0.0, Bottom:=0.0}
                                                            #Weapon icon BG
                                                            Widget:=texture_block{DefaultImage:=Textures.General.T_GenericSquare, DefaultDesiredSize:=vector2{X:=78.0, Y:=78.0}, DefaultTint:=MakeColorFromHex("#000c1d88")}

                                                        overlay_slot:
                                                            VerticalAlignment:=vertical_alignment.Center
                                                            HorizontalAlignment:=horizontal_alignment.Center
                                                            Padding:=margin{Left:=0.0, Top:=0.0, Right:=0.0, Bottom:=0.0}
                                                            #Weapon icon Border
                                                            Widget:=texture_block{DefaultImage:=Textures.General.T_GenericBorderSquare, DefaultDesiredSize:=vector2{X:=78.0, Y:=78.0}, DefaultTint:=MakeColorFromHex("#2e7be3FF")}

                                                        #overlay_slot:
                                                            #VerticalAlignment:=vertical_alignment.Top
                                                            #HorizontalAlignment:=horizontal_alignment.Center
                                                            #Padding:=margin{Left:=0.0, Top:=5.0, Right:=0.0, Bottom:=0.0}
                                                            #Weapon icon - Sniper
                                                            #Widget:=texture_block{DefaultImage:=Textures.Items.T_IconSniper, DefaultDesiredSize:=vector2{X:=50.0, Y:=50.0}, DefaultTint:=MakeColorFromHex("#FFFFFFFF")}

                                                        #overlay_slot:
                                                            #VerticalAlignment:=vertical_alignment.Bottom
                                                            #HorizontalAlignment:=horizontal_alignment.Center
                                                            #Padding:=margin{Left:=0.0, Top:=0.0, Right:=0.0, Bottom:=5.0}
                                                            #Weapon header - DMR Sniper
                                                            #Widget:=texture_block{DefaultImage:=Textures.Items.T_HeaderDMRSniper, DefaultDesiredSize:=vector2{X:=78.0, Y:=18.0}, DefaultTint:=MakeColorFromHex("#FFFFFFFF")}

                                            stack_box_slot:
                                                VerticalAlignment:=vertical_alignment.Fill
                                                HorizontalAlignment:=horizontal_alignment.Center #Center
                                                Padding:=margin{Left:=0.0, Top:=0.0, Right:=0.0, Bottom:=0.0}
                                                Widget:=overlay:
                                                    Slots:=array:
                                                        overlay_slot:
                                                            #Makeshift size box
                                                            VerticalAlignment:=vertical_alignment.Center
                                                            HorizontalAlignment:=horizontal_alignment.Center
                                                            Padding:=margin{Left:=0.0, Top:=0.0, Right:=0.0, Bottom:=0.0}
                                                            Widget:=texture_block{DefaultImage:=Textures.General.T_GenSquTrans, DefaultDesiredSize:=vector2{X:=418.0, Y:=100.0}, DefaultTint:=MakeColorFromHex("#00FF0011")}

                                                        overlay_slot:
                                                            #Vertical stack box to add all the rarity information
                                                            VerticalAlignment:=vertical_alignment.Center
                                                            HorizontalAlignment:=horizontal_alignment.Center
                                                            Padding:=margin{Left:=0.0, Top:=0.0, Right:=0.0, Bottom:=0.0}
                                                            Widget:=stack_box:
                                                                Orientation:=orientation.Vertical
                                                                Slots:=array:
                                                                    stack_box_slot:
                                                                        #Horizontal Stack box for max rarity available
                                                                        VerticalAlignment:=vertical_alignment.Top
                                                                        HorizontalAlignment:=horizontal_alignment.Center
                                                                        Padding:=margin{Left:=0.0, Top:=0.0, Right:=0.0, Bottom:=0.0}
                                                                        Widget:=stack_box:
                                                                            Orientation:=orientation.Horizontal
                                                                            Slots:=array:
                                                                                stack_box_slot:
                                                                                    VerticalAlignment:=vertical_alignment.Center
                                                                                    HorizontalAlignment:=horizontal_alignment.Center
                                                                                    Padding:=margin{Left:=0.0, Top:=0.0, Right:=0.0, Bottom:=0.0}
                                                                                    Widget:=text_block{DefaultText := StringToMessage("Rarity chance up to "), DefaultTextColor:=MakeColorFromHex("#FFFFFFFF")}

                                                                                stack_box_slot:
                                                                                    VerticalAlignment:=vertical_alignment.Center
                                                                                    HorizontalAlignment:=horizontal_alignment.Center
                                                                                    Padding:=margin{Left:=0.0, Top:=0.0, Right:=0.0, Bottom:=0.0}
                                                                                    Widget:=texture_block{DefaultImage:=Textures.Rarity.T_FlagCommon, DefaultDesiredSize:=vector2{X:=104.0, Y:=36.0}, DefaultTint:=MakeColorFromHex("#FFFFFFFF")}

                                                                    stack_box_slot:
                                                                        #Horizontal overlay for overall rarity scheme
                                                                        VerticalAlignment:=vertical_alignment.Bottom
                                                                        HorizontalAlignment:=horizontal_alignment.Center
                                                                        Padding:=margin{Left:=0.0, Top:=0.0, Right:=0.0, Bottom:=0.0}
                                                                        Widget:=overlay:
                                                                            Slots:=array:
                                                                                
                                                                                #Makeshift size box for the rarity scheme
                                                                                overlay_slot:
                                                                                    VerticalAlignment:=vertical_alignment.Center
                                                                                    HorizontalAlignment:=horizontal_alignment.Center
                                                                                    Padding:=margin{Left:=0.0, Top:=0.0, Right:=0.0, Bottom:=0.0}
                                                                                    Widget:=texture_block{DefaultImage:=Textures.General.T_GenSquTrans, DefaultDesiredSize:=vector2{X:=255.0, Y:=50.0}, DefaultTint:=MakeColorFromHex("#FF000022")}#Red frame

                                                                                #Visible border for rarity layout
                                                                                overlay_slot:
                                                                                    VerticalAlignment:=vertical_alignment.Center
                                                                                    HorizontalAlignment:=horizontal_alignment.Center
                                                                                    Padding:=margin{Left:=0.0, Top:=0.0, Right:=0.0, Bottom:=0.0}
                                                                                    Widget:=texture_block{DefaultImage:=Textures.General.T_GenericBorderSquare3, DefaultDesiredSize:=vector2{X:=255.0, Y:=40.0}, DefaultTint:=MakeColorFromHex("#2e7be3FF")}

                                                                                #Horizontal Stack box for rarity slots
                                                                                overlay_slot:
                                                                                    VerticalAlignment:=vertical_alignment.Fill
                                                                                    HorizontalAlignment:=horizontal_alignment.Fill
                                                                                    Padding:=margin{Left:=5.0, Top:=0.0, Right:=5.0, Bottom:=0.0}
                                                                                    Widget:=stack_box:
                                                                                        Orientation:=orientation.Horizontal
                                                                                        Slots:=array:
                                                                                            #Overlay for each rarity slots, including the pointing arrow
                                                                                            stack_box_slot:
                                                                                                VerticalAlignment:=vertical_alignment.Center
                                                                                                HorizontalAlignment:=horizontal_alignment.Center
                                                                                                Distribution:=option{1.0}
                                                                                                Padding:=margin{Left:=0.0, Top:=0.0, Right:=0.0, Bottom:=0.0}
                                                                                                Widget:=overlay:
                                                                                                    Slots:=array:
                                                                                                        #Makeshift size box
                                                                                                        overlay_slot:
                                                                                                            VerticalAlignment:=vertical_alignment.Center
                                                                                                            HorizontalAlignment:=horizontal_alignment.Center
                                                                                                            Padding:=margin{Left:=0.0, Top:=0.0, Right:=0.0, Bottom:=0.0}
                                                                                                            Widget:=texture_block{DefaultImage:=Textures.General.T_GenSquTrans, DefaultDesiredSize:=vector2{X:=30.0, Y:=40.0}, DefaultTint:=MakeColorFromHex("#0000FF08")}

                                                                                                        #Overlay for rarity border and texture
                                                                                                        overlay_slot:
                                                                                                            VerticalAlignment:=vertical_alignment.Center
                                                                                                            HorizontalAlignment:=horizontal_alignment.Center
                                                                                                            Padding:=margin{Left:=0.0, Top:=0.0, Right:=0.0, Bottom:=0.0}
                                                                                                            Widget:=overlay:
                                                                                                                Slots:=array:
                                                                                                                    
                                                                                                                    #Rarity texture border, when the rarity slot is empty
                                                                                                                    #overlay_slot:
                                                                                                                        #VerticalAlignment:=vertical_alignment.Center
                                                                                                                        #HorizontalAlignment:=horizontal_alignment.Center
                                                                                                                        #Padding:=margin{Left:=0.0, Top:=0.0, Right:=0.0, Bottom:=0.0}
                                                                                                                        #Widget:=texture_block{DefaultImage:=Textures.General.T_GenericBorderSquare2, DefaultDesiredSize:=vector2{X:=30.0, Y:=30.0}, DefaultTint:=MakeColorFromHex("#2e7be388")}

                                                                                                                    #Rarity texture
                                                                                                                    overlay_slot:
                                                                                                                        VerticalAlignment:=vertical_alignment.Center
                                                                                                                        HorizontalAlignment:=horizontal_alignment.Center
                                                                                                                        Padding:=margin{Left:=0.0, Top:=0.0, Right:=0.0, Bottom:=0.0}
                                                                                                                        Widget:=texture_block{DefaultImage:=Textures.General.T_GenericSquare, DefaultDesiredSize:=vector2{X:=20.0, Y:=20.0}, DefaultTint:=MakeColorFromHex("#979797FF")}
                                                                                                        
                                                                                                                    

                                                                                                        #Bracket
                                                                                                        overlay_slot:
                                                                                                            VerticalAlignment:=vertical_alignment.Fill
                                                                                                            HorizontalAlignment:=horizontal_alignment.Fill
                                                                                                            Padding:=margin{Left:=0.0, Top:=0.0, Right:=0.0, Bottom:=0.0}
                                                                                                            Widget:=stack_box:
                                                                                                                Orientation:=orientation.Horizontal
                                                                                                                #two instances of bracket texture to represent the range of rarities
                                                                                                                Slots:=array:

                                                                                                                    stack_box_slot:
                                                                                                                        VerticalAlignment:=vertical_alignment.Center
                                                                                                                        HorizontalAlignment:=horizontal_alignment.Left
                                                                                                                        Distribution:=option{1.0}
                                                                                                                        Padding:=margin{Left:=0.0, Top:=0.0, Right:=0.0, Bottom:=0.0}
                                                                                                                        Widget:=texture_block{DefaultImage:=Textures.General.T_RarityBracketLeft, DefaultDesiredSize:=vector2{X:=18.0, Y:=36.0}, DefaultTint:=MakeColorFromHex("#FFFFFF44")}

                                                                                                                    stack_box_slot:
                                                                                                                        VerticalAlignment:=vertical_alignment.Center
                                                                                                                        HorizontalAlignment:=horizontal_alignment.Right
                                                                                                                        Distribution:=option{1.0}
                                                                                                                        Padding:=margin{Left:=0.0, Top:=0.0, Right:=0.0, Bottom:=0.0}
                                                                                                                        Widget:=texture_block{DefaultImage:=Textures.General.T_RarityBracketRight, DefaultDesiredSize:=vector2{X:=18.0, Y:=36.0}, DefaultTint:=MakeColorFromHex("#FFFFFF44")}

                                                                                            stack_box_slot:
                                                                                                VerticalAlignment:=vertical_alignment.Center
                                                                                                HorizontalAlignment:=horizontal_alignment.Center
                                                                                                Distribution:=option{1.0}
                                                                                                Padding:=margin{Left:=0.0, Top:=0.0, Right:=0.0, Bottom:=0.0}
                                                                                                Widget:=overlay:
                                                                                                    Slots:=array:
                                                                                                        #Makeshift size box
                                                                                                        overlay_slot:
                                                                                                            VerticalAlignment:=vertical_alignment.Center
                                                                                                            HorizontalAlignment:=horizontal_alignment.Center
                                                                                                            Padding:=margin{Left:=0.0, Top:=0.0, Right:=0.0, Bottom:=0.0}
                                                                                                            Widget:=texture_block{DefaultImage:=Textures.General.T_GenSquTrans, DefaultDesiredSize:=vector2{X:=30.0, Y:=40.0}, DefaultTint:=MakeColorFromHex("#0000FF08")}

                                                                                                        #Overlay for rarity border and texture
                                                                                                        overlay_slot:
                                                                                                            VerticalAlignment:=vertical_alignment.Center
                                                                                                            HorizontalAlignment:=horizontal_alignment.Center
                                                                                                            Padding:=margin{Left:=0.0, Top:=0.0, Right:=0.0, Bottom:=0.0}
                                                                                                            Widget:=overlay:
                                                                                                                Slots:=array:
                                                                                                                    
                                                                                                                    #Rarity texture border, when the rarity slot is empty
                                                                                                                    overlay_slot:
                                                                                                                        VerticalAlignment:=vertical_alignment.Center
                                                                                                                        HorizontalAlignment:=horizontal_alignment.Center
                                                                                                                        Padding:=margin{Left:=0.0, Top:=0.0, Right:=0.0, Bottom:=0.0}
                                                                                                                        Widget:=texture_block{DefaultImage:=Textures.General.T_GenericBorderSquare2, DefaultDesiredSize:=vector2{X:=30.0, Y:=30.0}, DefaultTint:=MakeColorFromHex("#2e7be388")}

                                                                                                                    #Rarity texture
                                                                                                                    #overlay_slot:
                                                                                                                        #VerticalAlignment:=vertical_alignment.Center
                                                                                                                        #HorizontalAlignment:=horizontal_alignment.Center
                                                                                                                        #Padding:=margin{Left:=0.0, Top:=0.0, Right:=0.0, Bottom:=0.0}
                                                                                                                        #Widget:=texture_block{DefaultImage:=Textures.General.T_GenericSquare, DefaultDesiredSize:=vector2{X:=30.0, Y:=30.0}, DefaultTint:=MakeColorFromHex("#4ccc2aFF")}

                                                                                                        #Bracket
                                                                                                        #overlay_slot:
                                                                                                            #VerticalAlignment:=vertical_alignment.Fill
                                                                                                            #HorizontalAlignment:=horizontal_alignment.Fill
                                                                                                            #Padding:=margin{Left:=0.0, Top:=0.0, Right:=0.0, Bottom:=0.0}
                                                                                                            #Widget:=stack_box:
                                                                                                                #Orientation:=orientation.Horizontal
                                                                                                                #two instances of bracket texture to represent the range of rarities
                                                                                                                #Slots:=array:

                                                                                                                    #stack_box_slot:
                                                                                                                        #VerticalAlignment:=vertical_alignment.Center
                                                                                                                        #HorizontalAlignment:=horizontal_alignment.Left
                                                                                                                        #Distribution:=option{1.0}
                                                                                                                        #Padding:=margin{Left:=0.0, Top:=0.0, Right:=0.0, Bottom:=0.0}
                                                                                                                        #Widget:=texture_block{DefaultImage:=Textures.General.T_RarityBracketLeft, DefaultDesiredSize:=vector2{X:=18.0, Y:=36.0}, DefaultTint:=MakeColorFromHex("#FFFFFF44")}

                                                                                                                    #stack_box_slot:
                                                                                                                        #VerticalAlignment:=vertical_alignment.Center
                                                                                                                        #HorizontalAlignment:=horizontal_alignment.Right
                                                                                                                        #Distribution:=option{1.0}
                                                                                                                        #Padding:=margin{Left:=0.0, Top:=0.0, Right:=0.0, Bottom:=0.0}
                                                                                                                        #Widget:=texture_block{DefaultImage:=Textures.General.T_RarityBracketRight, DefaultDesiredSize:=vector2{X:=18.0, Y:=36.0}, DefaultTint:=MakeColorFromHex("#FFFFFF44")}

                                                                                            stack_box_slot:
                                                                                                VerticalAlignment:=vertical_alignment.Center
                                                                                                HorizontalAlignment:=horizontal_alignment.Center
                                                                                                Distribution:=option{1.0}
                                                                                                Padding:=margin{Left:=0.0, Top:=0.0, Right:=0.0, Bottom:=0.0}
                                                                                                Widget:=overlay:
                                                                                                    Slots:=array:
                                                                                                        #Makeshift size box
                                                                                                        overlay_slot:
                                                                                                            VerticalAlignment:=vertical_alignment.Center
                                                                                                            HorizontalAlignment:=horizontal_alignment.Center
                                                                                                            Padding:=margin{Left:=0.0, Top:=0.0, Right:=0.0, Bottom:=0.0}
                                                                                                            Widget:=texture_block{DefaultImage:=Textures.General.T_GenSquTrans, DefaultDesiredSize:=vector2{X:=30.0, Y:=40.0}, DefaultTint:=MakeColorFromHex("#0000FF08")}

                                                                                                        #Overlay for rarity border and texture
                                                                                                        overlay_slot:
                                                                                                            VerticalAlignment:=vertical_alignment.Center
                                                                                                            HorizontalAlignment:=horizontal_alignment.Center
                                                                                                            Padding:=margin{Left:=0.0, Top:=0.0, Right:=0.0, Bottom:=0.0}
                                                                                                            Widget:=overlay:
                                                                                                                Slots:=array:
                                                                                                                                                                                                                
                                                                                                                    #Rarity texture border, when the rarity slot is empty
                                                                                                                    overlay_slot:
                                                                                                                        VerticalAlignment:=vertical_alignment.Center
                                                                                                                        HorizontalAlignment:=horizontal_alignment.Center
                                                                                                                        Padding:=margin{Left:=0.0, Top:=0.0, Right:=0.0, Bottom:=0.0}
                                                                                                                        Widget:=texture_block{DefaultImage:=Textures.General.T_GenericBorderSquare2, DefaultDesiredSize:=vector2{X:=30.0, Y:=30.0}, DefaultTint:=MakeColorFromHex("#2e7be388")}
                                                                                                                                                                                                                            
                                                                                                                    #Rarity texture
                                                                                                                    #overlay_slot:
                                                                                                                        #VerticalAlignment:=vertical_alignment.Bottom
                                                                                                                        #HorizontalAlignment:=horizontal_alignment.Center
                                                                                                                        #Padding:=margin{Left:=0.0, Top:=0.0, Right:=0.0, Bottom:=0.0}
                                                                                                                        #Widget:=texture_block{DefaultImage:=Textures.General.T_GenericSquare, DefaultDesiredSize:=vector2{X:=30.0, Y:=30.0}, DefaultTint:=MakeColorFromHex("#2a54ccFF")}
                                                                                                                    
                                                                                                        #Bracket
                                                                                                        #overlay_slot:
                                                                                                            #VerticalAlignment:=vertical_alignment.Fill
                                                                                                            #HorizontalAlignment:=horizontal_alignment.Fill
                                                                                                            #Padding:=margin{Left:=0.0, Top:=0.0, Right:=0.0, Bottom:=0.0}
                                                                                                            #Widget:=stack_box:
                                                                                                                #Orientation:=orientation.Horizontal
                                                                                                                #two instances of bracket texture to represent the range of rarities
                                                                                                                #Slots:=array:

                                                                                                                    #stack_box_slot:
                                                                                                                        #VerticalAlignment:=vertical_alignment.Center
                                                                                                                        #HorizontalAlignment:=horizontal_alignment.Left
                                                                                                                        #Distribution:=option{1.0}
                                                                                                                        #Padding:=margin{Left:=0.0, Top:=0.0, Right:=0.0, Bottom:=0.0}
                                                                                                                        #Widget:=texture_block{DefaultImage:=Textures.General.T_RarityBracketLeft, DefaultDesiredSize:=vector2{X:=18.0, Y:=36.0}, DefaultTint:=MakeColorFromHex("#FFFFFF44")}

                                                                                                                    #stack_box_slot:
                                                                                                                        #VerticalAlignment:=vertical_alignment.Center
                                                                                                                        #HorizontalAlignment:=horizontal_alignment.Right
                                                                                                                        #Distribution:=option{1.0}
                                                                                                                        #Padding:=margin{Left:=0.0, Top:=0.0, Right:=0.0, Bottom:=0.0}
                                                                                                                        #Widget:=texture_block{DefaultImage:=Textures.General.T_RarityBracketRight, DefaultDesiredSize:=vector2{X:=18.0, Y:=36.0}, DefaultTint:=MakeColorFromHex("#FFFFFF44")}

                                                                                            stack_box_slot:
                                                                                                VerticalAlignment:=vertical_alignment.Center
                                                                                                HorizontalAlignment:=horizontal_alignment.Center
                                                                                                Distribution:=option{1.0}
                                                                                                Padding:=margin{Left:=0.0, Top:=0.0, Right:=0.0, Bottom:=0.0}
                                                                                                Widget:=overlay:
                                                                                                    Slots:=array:
                                                                                                        #Makeshift size box
                                                                                                        overlay_slot:
                                                                                                            VerticalAlignment:=vertical_alignment.Center
                                                                                                            HorizontalAlignment:=horizontal_alignment.Center
                                                                                                            Padding:=margin{Left:=0.0, Top:=0.0, Right:=0.0, Bottom:=0.0}
                                                                                                            Widget:=texture_block{DefaultImage:=Textures.General.T_GenSquTrans, DefaultDesiredSize:=vector2{X:=30.0, Y:=40.0}, DefaultTint:=MakeColorFromHex("#0000FF08")}

                                                                                                        #Overlay for rarity border and texture
                                                                                                        overlay_slot:
                                                                                                            VerticalAlignment:=vertical_alignment.Center
                                                                                                            HorizontalAlignment:=horizontal_alignment.Center
                                                                                                            Padding:=margin{Left:=0.0, Top:=0.0, Right:=0.0, Bottom:=0.0}
                                                                                                            Widget:=overlay:
                                                                                                                Slots:=array:

                                                                                                                    #Rarity texture border, when the rarity slot is empty
                                                                                                                    overlay_slot:
                                                                                                                        VerticalAlignment:=vertical_alignment.Center
                                                                                                                        HorizontalAlignment:=horizontal_alignment.Center
                                                                                                                        Padding:=margin{Left:=0.0, Top:=0.0, Right:=0.0, Bottom:=0.0}
                                                                                                                        Widget:=texture_block{DefaultImage:=Textures.General.T_GenericBorderSquare2, DefaultDesiredSize:=vector2{X:=30.0, Y:=30.0}, DefaultTint:=MakeColorFromHex("#2e7be388")}
                                                                                                                    
                                                                                                                    #Rarity texture
                                                                                                                    #overlay_slot:
                                                                                                                        #VerticalAlignment:=vertical_alignment.Bottom
                                                                                                                        #HorizontalAlignment:=horizontal_alignment.Center
                                                                                                                        #Padding:=margin{Left:=0.0, Top:=0.0, Right:=0.0, Bottom:=0.0}
                                                                                                                        #Widget:=texture_block{DefaultImage:=Textures.General.T_GenericSquare, DefaultDesiredSize:=vector2{X:=30.0, Y:=30.0}, DefaultTint:=MakeColorFromHex("#b32accFF")}
                                                                                                                    
                                                                                                        #Bracket
                                                                                                        #overlay_slot:
                                                                                                            #VerticalAlignment:=vertical_alignment.Fill
                                                                                                            #HorizontalAlignment:=horizontal_alignment.Fill
                                                                                                            #Padding:=margin{Left:=0.0, Top:=0.0, Right:=0.0, Bottom:=0.0}
                                                                                                            #Widget:=stack_box:
                                                                                                                #Orientation:=orientation.Horizontal
                                                                                                                #two instances of bracket texture to represent the range of rarities
                                                                                                                #Slots:=array:

                                                                                                                    #stack_box_slot:
                                                                                                                        #VerticalAlignment:=vertical_alignment.Center
                                                                                                                        #HorizontalAlignment:=horizontal_alignment.Left
                                                                                                                        #Distribution:=option{1.0}
                                                                                                                        #Padding:=margin{Left:=0.0, Top:=0.0, Right:=0.0, Bottom:=0.0}
                                                                                                                        #Widget:=texture_block{DefaultImage:=Textures.General.T_RarityBracketLeft, DefaultDesiredSize:=vector2{X:=18.0, Y:=36.0}, DefaultTint:=MakeColorFromHex("#FFFFFF44")}

                                                                                                                    #stack_box_slot:
                                                                                                                        #VerticalAlignment:=vertical_alignment.Center
                                                                                                                        #HorizontalAlignment:=horizontal_alignment.Right
                                                                                                                        #Distribution:=option{1.0}
                                                                                                                        #Padding:=margin{Left:=0.0, Top:=0.0, Right:=0.0, Bottom:=0.0}
                                                                                                                        #Widget:=texture_block{DefaultImage:=Textures.General.T_RarityBracketRight, DefaultDesiredSize:=vector2{X:=18.0, Y:=36.0}, DefaultTint:=MakeColorFromHex("#FFFFFF44")}

                                                                                            stack_box_slot:
                                                                                                VerticalAlignment:=vertical_alignment.Center
                                                                                                HorizontalAlignment:=horizontal_alignment.Center
                                                                                                Distribution:=option{1.0}
                                                                                                Padding:=margin{Left:=0.0, Top:=0.0, Right:=0.0, Bottom:=0.0}
                                                                                                Widget:=overlay:
                                                                                                    Slots:=array:
                                                                                                        #Makeshift size box
                                                                                                        overlay_slot:
                                                                                                            VerticalAlignment:=vertical_alignment.Center
                                                                                                            HorizontalAlignment:=horizontal_alignment.Center
                                                                                                            Padding:=margin{Left:=0.0, Top:=0.0, Right:=0.0, Bottom:=0.0}
                                                                                                            Widget:=texture_block{DefaultImage:=Textures.General.T_GenSquTrans, DefaultDesiredSize:=vector2{X:=30.0, Y:=40.0}, DefaultTint:=MakeColorFromHex("#0000FF08")}

                                                                                                        #Overlay for rarity border and texture
                                                                                                        overlay_slot:
                                                                                                            VerticalAlignment:=vertical_alignment.Center
                                                                                                            HorizontalAlignment:=horizontal_alignment.Center
                                                                                                            Padding:=margin{Left:=0.0, Top:=0.0, Right:=0.0, Bottom:=0.0}
                                                                                                            Widget:=overlay:
                                                                                                                Slots:=array:

                                                                                                                    #Rarity texture border, when the rarity slot is empty
                                                                                                                    overlay_slot:
                                                                                                                        VerticalAlignment:=vertical_alignment.Center
                                                                                                                        HorizontalAlignment:=horizontal_alignment.Center
                                                                                                                        Padding:=margin{Left:=0.0, Top:=0.0, Right:=0.0, Bottom:=0.0}
                                                                                                                        Widget:=texture_block{DefaultImage:=Textures.General.T_GenericBorderSquare2, DefaultDesiredSize:=vector2{X:=30.0, Y:=30.0}, DefaultTint:=MakeColorFromHex("#2e7be388")}
                                                                                                                    
                                                                                                                    #Rarity texture
                                                                                                                    #overlay_slot:
                                                                                                                        #VerticalAlignment:=vertical_alignment.Bottom
                                                                                                                        #HorizontalAlignment:=horizontal_alignment.Center
                                                                                                                        #Padding:=margin{Left:=0.0, Top:=0.0, Right:=0.0, Bottom:=0.0}
                                                                                                                        #Widget:=texture_block{DefaultImage:=Textures.General.T_GenericSquare, DefaultDesiredSize:=vector2{X:=30.0, Y:=30.0}, DefaultTint:=MakeColorFromHex("#cc8d2aFF")}

                                                                                                        #Bracket
                                                                                                        #overlay_slot:
                                                                                                            #VerticalAlignment:=vertical_alignment.Fill
                                                                                                            #HorizontalAlignment:=horizontal_alignment.Fill
                                                                                                            #Padding:=margin{Left:=0.0, Top:=0.0, Right:=0.0, Bottom:=0.0}
                                                                                                            #Widget:=stack_box:
                                                                                                                #Orientation:=orientation.Horizontal
                                                                                                                #two instances of bracket texture to represent the range of rarities
                                                                                                                #Slots:=array:

                                                                                                                    #stack_box_slot:
                                                                                                                        #VerticalAlignment:=vertical_alignment.Center
                                                                                                                        #HorizontalAlignment:=horizontal_alignment.Left
                                                                                                                        #Distribution:=option{1.0}
                                                                                                                        #Padding:=margin{Left:=0.0, Top:=0.0, Right:=0.0, Bottom:=0.0}
                                                                                                                        #Widget:=texture_block{DefaultImage:=Textures.General.T_RarityBracketLeft, DefaultDesiredSize:=vector2{X:=18.0, Y:=36.0}, DefaultTint:=MakeColorFromHex("#FFFFFF44")}

                                                                                                                    #stack_box_slot:
                                                                                                                        #VerticalAlignment:=vertical_alignment.Center
                                                                                                                        #HorizontalAlignment:=horizontal_alignment.Right
                                                                                                                        #Distribution:=option{1.0}
                                                                                                                        #Padding:=margin{Left:=0.0, Top:=0.0, Right:=0.0, Bottom:=0.0}
                                                                                                                        #Widget:=texture_block{DefaultImage:=Textures.General.T_RarityBracketRight, DefaultDesiredSize:=vector2{X:=18.0, Y:=36.0}, DefaultTint:=MakeColorFromHex("#FFFFFF44")}

                                                                                            stack_box_slot:
                                                                                                VerticalAlignment:=vertical_alignment.Center
                                                                                                HorizontalAlignment:=horizontal_alignment.Center
                                                                                                Distribution:=option{1.0}
                                                                                                Padding:=margin{Left:=0.0, Top:=0.0, Right:=0.0, Bottom:=0.0}
                                                                                                Widget:=overlay:
                                                                                                    Slots:=array:
                                                                                                        #Makeshift size box
                                                                                                        overlay_slot:
                                                                                                            VerticalAlignment:=vertical_alignment.Center
                                                                                                            HorizontalAlignment:=horizontal_alignment.Center
                                                                                                            Padding:=margin{Left:=0.0, Top:=0.0, Right:=0.0, Bottom:=0.0}
                                                                                                            Widget:=texture_block{DefaultImage:=Textures.General.T_GenSquTrans, DefaultDesiredSize:=vector2{X:=30.0, Y:=40.0}, DefaultTint:=MakeColorFromHex("#0000FF08")}

                                                                                                        #Overlay for rarity border and texture
                                                                                                        overlay_slot:
                                                                                                            VerticalAlignment:=vertical_alignment.Center
                                                                                                            HorizontalAlignment:=horizontal_alignment.Center
                                                                                                            Padding:=margin{Left:=0.0, Top:=0.0, Right:=0.0, Bottom:=0.0}
                                                                                                            Widget:=overlay:
                                                                                                                Slots:=array:
                                                                                                        
                                                                                                                    #Rarity texture border, when the rarity slot is empty
                                                                                                                    overlay_slot:
                                                                                                                        VerticalAlignment:=vertical_alignment.Center
                                                                                                                        HorizontalAlignment:=horizontal_alignment.Center
                                                                                                                        Padding:=margin{Left:=0.0, Top:=0.0, Right:=0.0, Bottom:=0.0}
                                                                                                                        Widget:=texture_block{DefaultImage:=Textures.General.T_GenericBorderSquare2, DefaultDesiredSize:=vector2{X:=30.0, Y:=30.0}, DefaultTint:=MakeColorFromHex("#2e7be388")}
                                                                                                                    
                                                                                                                    #Rarity texture
                                                                                                                    #overlay_slot:
                                                                                                                        #VerticalAlignment:=vertical_alignment.Bottom
                                                                                                                        #HorizontalAlignment:=horizontal_alignment.Center
                                                                                                                        #Padding:=margin{Left:=0.0, Top:=0.0, Right:=0.0, Bottom:=0.0}
                                                                                                                        #Widget:=texture_block{DefaultImage:=Textures.General.T_GenericSquare, DefaultDesiredSize:=vector2{X:=30.0, Y:=30.0}, DefaultTint:=MakeColorFromHex("#eacf19FF")}
                                                                                                                    
                                                                                                        #Bracket
                                                                                                        #overlay_slot:
                                                                                                            #VerticalAlignment:=vertical_alignment.Fill
                                                                                                            #HorizontalAlignment:=horizontal_alignment.Fill
                                                                                                            #Padding:=margin{Left:=0.0, Top:=0.0, Right:=0.0, Bottom:=0.0}
                                                                                                            #Widget:=stack_box:
                                                                                                                #Orientation:=orientation.Horizontal
                                                                                                                #two instances of bracket texture to represent the range of rarities
                                                                                                                #Slots:=array:

                                                                                                                    #stack_box_slot:
                                                                                                                        #VerticalAlignment:=vertical_alignment.Center
                                                                                                                        #HorizontalAlignment:=horizontal_alignment.Left
                                                                                                                        #Distribution:=option{1.0}
                                                                                                                        #Padding:=margin{Left:=0.0, Top:=0.0, Right:=0.0, Bottom:=0.0}
                                                                                                                        #Widget:=texture_block{DefaultImage:=Textures.General.T_RarityBracketLeft, DefaultDesiredSize:=vector2{X:=18.0, Y:=36.0}, DefaultTint:=MakeColorFromHex("#FFFFFF44")}

                                                                                                                    #stack_box_slot:
                                                                                                                        #VerticalAlignment:=vertical_alignment.Center
                                                                                                                        #HorizontalAlignment:=horizontal_alignment.Right
                                                                                                                        #Distribution:=option{1.0}
                                                                                                                        #Padding:=margin{Left:=0.0, Top:=0.0, Right:=0.0, Bottom:=0.0}
                                                                                                                        #Widget:=texture_block{DefaultImage:=Textures.General.T_RarityBracketRight, DefaultDesiredSize:=vector2{X:=18.0, Y:=36.0}, DefaultTint:=MakeColorFromHex("#FFFFFF44")}

                    #overlay_slot:
                        #VerticalAlignment:=vertical_alignment.Fill
                        #HorizontalAlignment:=horizontal_alignment.Fill
                        #Padding:=margin{Left:=0.0, Top:=0.0, Right:=0.0, Bottom:=0.0}
                        #Widget:=texture_block{DefaultImage:=Textures.General.T_LevelPointer, DefaultDesiredSize:=vector2{X:=20.0, Y:=20.0}, DefaultTint:=MakeColorFromHex("#7de2FFFF")}
                    
            #Definition for the level row 1
            LevelRow1:=overlay:
                Slots:=array:
                    overlay_slot:
                        #Makeshift size box
                        VerticalAlignment:=vertical_alignment.Center
                        HorizontalAlignment:=horizontal_alignment.Center
                        Padding:=margin{Left:=0.0, Top:=0.0, Right:=0.0, Bottom:=0.0}
                        Widget:=texture_block{DefaultImage:=Textures.General.T_GenSquTrans, DefaultDesiredSize:=vector2{X:=1250.0, Y:=114.0}, DefaultTint:=MakeColorFromHex("#0000FFFF")}

                    overlay_slot:
                        #Rarity level content
                        VerticalAlignment:=vertical_alignment.Center
                        HorizontalAlignment:=horizontal_alignment.Left
                        Padding:=margin{Left:=0.0, Top:=0.0, Right:=0.0, Bottom:=0.0}
                        Widget:=overlay:
                            Slots:=array:

                                overlay_slot:
                                    #Makeshift size box
                                    VerticalAlignment:=vertical_alignment.Center
                                    HorizontalAlignment:=horizontal_alignment.Center
                                    Padding:=margin{Left:=0.0, Top:=0.0, Right:=0.0, Bottom:=0.0}
                                    Widget:=texture_block{DefaultImage:=Textures.General.T_GenericSquare, DefaultDesiredSize:=vector2{X:=1190.0, Y:=114.0}, DefaultTint:=MakeColorFromHex("#1f4e7f32")}

                                overlay_slot:
                                    #Highlight texture, over the content but under the rarity flag
                                    VerticalAlignment:=vertical_alignment.Fill
                                    HorizontalAlignment:=horizontal_alignment.Fill
                                    Padding:=margin{Left:=0.0, Top:=0.0, Right:=0.0, Bottom:=0.0}
                                    Widget:=texture_block{DefaultImage:=Textures.General.T_LevelPointer, DefaultDesiredSize:=vector2{X:=512.0, Y:=64.0}, DefaultTint:=MakeColorFromHex("#7de2FFFF")}

                                    #Horizontal stack box
                                overlay_slot:
                                    VerticalAlignment:=vertical_alignment.Fill
                                    HorizontalAlignment:=horizontal_alignment.Left
                                    Padding:=margin{Left:=0.0, Top:=0.0, Right:=0.0, Bottom:=0.0}
                                    Widget:=stack_box:
                                        Orientation:=orientation.Horizontal
                                        Slots:=array:
                                            #Level flag
                                            stack_box_slot:
                                                VerticalAlignment:=vertical_alignment.Top
                                                HorizontalAlignment:=horizontal_alignment.Left
                                                Padding:=margin{Left:=0.0, Top:=0.0, Right:=15.0, Bottom:=0.0}
                                                Widget:=overlay:

                                                    Slots:=array:
                                                        overlay_slot:
                                                            VerticalAlignment:=vertical_alignment.Center
                                                            HorizontalAlignment:=horizontal_alignment.Center
                                                            Padding:=margin{Left:=0.0, Top:=0.0, Right:=0.0, Bottom:=0.0}
                                                            #Tier level makeshift size box
                                                            Widget:=texture_block{DefaultImage:=Textures.General.T_GenSquTrans, DefaultDesiredSize:=vector2{X:=150.0, Y:=114.0}, DefaultTint:=MakeColorFromHex("#FF0000FF")}

                                                        overlay_slot:
                                                            VerticalAlignment:=vertical_alignment.Top
                                                            HorizontalAlignment:=horizontal_alignment.Left
                                                            Padding:=margin{Left:=0.0, Top:=0.0, Right:=0.0, Bottom:=0.0}
                                                            #Selected level yellow flag
                                                            Widget:=overlay:

                                                                Slots:=array:
                                                                    overlay_slot:
                                                                        VerticalAlignment:=vertical_alignment.Center
                                                                        HorizontalAlignment:=horizontal_alignment.Center
                                                                        Padding:=margin{Left:=0.0, Top:=0.0, Right:=0.0, Bottom:=0.0}
                                                                        #level flag texture
                                                                        Widget:=texture_block{DefaultImage:=Textures.General.T_GenLevelFlag, DefaultDesiredSize:=vector2{X:=140.0, Y:=45.0}, DefaultTint:=MakeColorFromHex("#f7ff1aFF")}

                                                                    overlay_slot:
                                                                        VerticalAlignment:=vertical_alignment.Center
                                                                        HorizontalAlignment:=horizontal_alignment.Center
                                                                        Padding:=margin{Left:=0.0, Top:=0.0, Right:=0.0, Bottom:=0.0}
                                                                        #level flag text
                                                                        Widget:=text_block{DefaultText := StringToMessage("Level 1"), DefaultTextColor:=MakeColorFromHex("#000000FF")}

                                                        #overlay_slot:
                                                            #VerticalAlignment:=vertical_alignment.Fill
                                                            #HorizontalAlignment:=horizontal_alignment.Center
                                                            #Padding:=margin{Left:=0.0, Top:=5.0, Right:=0.0, Bottom:=5.0}
                                                            #Unselected level and lock
                                                            #Widget:=overlay:
                                                                #Slots:=array:
                                                                    
                                                                    #overlay_slot:
                                                                        #VerticalAlignment:=vertical_alignment.Top
                                                                        #HorizontalAlignment:=horizontal_alignment.Center
                                                                        #Padding:=margin{Left:=0.0, Top:=0.0, Right:=0.0, Bottom:=0.0}
                                                                        #level flag texture
                                                                        #Widget:=text_block{DefaultText := StringToMessage("Level 0"), DefaultTextColor:=MakeColorFromHex("#FFFFFFFF")}

                                                                    #overlay_slot:
                                                                        #VerticalAlignment:=vertical_alignment.Bottom
                                                                        #HorizontalAlignment:=horizontal_alignment.Center
                                                                        #Padding:=margin{Left:=0.0, Top:=0.0, Right:=0.0, Bottom:=0.0}
                                                                        #level flag texture
                                                                        #Widget:=texture_block{DefaultImage:=Textures.General.T_Lock, DefaultDesiredSize:=vector2{X:=52.0, Y:=58.0}, DefaultTint:=MakeColorFromHex("#ffffff88")}
                                            
                                            #Level weapon roster
                                            stack_box_slot:
                                                VerticalAlignment:=vertical_alignment.Center
                                                HorizontalAlignment:=horizontal_alignment.Center
                                                Padding:=margin{Left:=0.0, Top:=0.0, Right:=8.0, Bottom:=0.0}
                                                Widget:=overlay:
                                                    Slots:=array:
                                                        #Weapon Icon - Pistols
                                                        overlay_slot:
                                                            VerticalAlignment:=vertical_alignment.Center
                                                            HorizontalAlignment:=horizontal_alignment.Center
                                                            Padding:=margin{Left:=0.0, Top:=0.0, Right:=0.0, Bottom:=0.0}
                                                            #Weapon icon BG
                                                            Widget:=texture_block{DefaultImage:=Textures.General.T_GenericSquare, DefaultDesiredSize:=vector2{X:=78.0, Y:=78.0}, DefaultTint:=MakeColorFromHex("#000c1d88")}

                                                        overlay_slot:
                                                            VerticalAlignment:=vertical_alignment.Center
                                                            HorizontalAlignment:=horizontal_alignment.Center
                                                            Padding:=margin{Left:=0.0, Top:=0.0, Right:=0.0, Bottom:=0.0}
                                                            #Weapon icon Border
                                                            Widget:=texture_block{DefaultImage:=Textures.General.T_GenericBorderSquare, DefaultDesiredSize:=vector2{X:=78.0, Y:=78.0}, DefaultTint:=MakeColorFromHex("#2e7be3FF")}

                                                        overlay_slot:
                                                            VerticalAlignment:=vertical_alignment.Top
                                                            HorizontalAlignment:=horizontal_alignment.Center
                                                            Padding:=margin{Left:=0.0, Top:=5.0, Right:=0.0, Bottom:=0.0}
                                                            #Weapon icon - pistol
                                                            Widget:=texture_block{DefaultImage:=Textures.Items.T_IconPistol, DefaultDesiredSize:=vector2{X:=50.0, Y:=50.0}, DefaultTint:=MakeColorFromHex("#FFFFFFFF")}

                                                        overlay_slot:
                                                            VerticalAlignment:=vertical_alignment.Bottom
                                                            HorizontalAlignment:=horizontal_alignment.Center
                                                            Padding:=margin{Left:=0.0, Top:=0.0, Right:=0.0, Bottom:=5.0}
                                                            #Weapon header - pistol
                                                            Widget:=texture_block{DefaultImage:=Textures.Items.T_HeaderPistol, DefaultDesiredSize:=vector2{X:=78.0, Y:=18.0}, DefaultTint:=MakeColorFromHex("#FFFFFFFF")}
                                                        
                                            stack_box_slot:
                                                VerticalAlignment:=vertical_alignment.Center
                                                HorizontalAlignment:=horizontal_alignment.Center
                                                Padding:=margin{Left:=0.0, Top:=0.0, Right:=8.0, Bottom:=0.0}
                                                Widget:=overlay:
                                                    Slots:=array:
                                                        #Weapon Icon - Shotgun
                                                        overlay_slot:
                                                            VerticalAlignment:=vertical_alignment.Center
                                                            HorizontalAlignment:=horizontal_alignment.Center
                                                            Padding:=margin{Left:=0.0, Top:=0.0, Right:=0.0, Bottom:=0.0}
                                                            #Weapon icon BG
                                                            Widget:=texture_block{DefaultImage:=Textures.General.T_GenericSquare, DefaultDesiredSize:=vector2{X:=78.0, Y:=78.0}, DefaultTint:=MakeColorFromHex("#000c1d88")}

                                                        overlay_slot:
                                                            VerticalAlignment:=vertical_alignment.Center
                                                            HorizontalAlignment:=horizontal_alignment.Center
                                                            Padding:=margin{Left:=0.0, Top:=0.0, Right:=0.0, Bottom:=0.0}
                                                            #Weapon icon Border
                                                            Widget:=texture_block{DefaultImage:=Textures.General.T_GenericBorderSquare, DefaultDesiredSize:=vector2{X:=78.0, Y:=78.0}, DefaultTint:=MakeColorFromHex("#2e7be3FF")}

                                                        overlay_slot:
                                                            VerticalAlignment:=vertical_alignment.Top
                                                            HorizontalAlignment:=horizontal_alignment.Center
                                                            Padding:=margin{Left:=0.0, Top:=5.0, Right:=0.0, Bottom:=0.0}
                                                            #Weapon icon - shotgun
                                                            Widget:=texture_block{DefaultImage:=Textures.Items.T_IconShotgun, DefaultDesiredSize:=vector2{X:=50.0, Y:=50.0}, DefaultTint:=MakeColorFromHex("#FFFFFFFF")}

                                                        overlay_slot:
                                                            VerticalAlignment:=vertical_alignment.Bottom
                                                            HorizontalAlignment:=horizontal_alignment.Center
                                                            Padding:=margin{Left:=0.0, Top:=0.0, Right:=0.0, Bottom:=5.0}
                                                            #Weapon header - shotgun
                                                            Widget:=texture_block{DefaultImage:=Textures.Items.T_HeaderShotgun, DefaultDesiredSize:=vector2{X:=78.0, Y:=18.0}, DefaultTint:=MakeColorFromHex("#FFFFFFFF")}

                                            stack_box_slot:
                                                VerticalAlignment:=vertical_alignment.Center
                                                HorizontalAlignment:=horizontal_alignment.Center
                                                Padding:=margin{Left:=0.0, Top:=0.0, Right:=8.0, Bottom:=0.0}
                                                Widget:=overlay:
                                                    Slots:=array:
                                                        #Weapon Icon - SMG
                                                        overlay_slot:
                                                            VerticalAlignment:=vertical_alignment.Top
                                                            HorizontalAlignment:=horizontal_alignment.Center
                                                            Padding:=margin{Left:=0.0, Top:=0.0, Right:=0.0, Bottom:=0.0}
                                                            #Weapon icon BG
                                                            Widget:=texture_block{DefaultImage:=Textures.General.T_GenericSquare, DefaultDesiredSize:=vector2{X:=78.0, Y:=78.0}, DefaultTint:=MakeColorFromHex("#000c1d88")}

                                                        overlay_slot:
                                                            VerticalAlignment:=vertical_alignment.Center
                                                            HorizontalAlignment:=horizontal_alignment.Center
                                                            Padding:=margin{Left:=0.0, Top:=0.0, Right:=0.0, Bottom:=0.0}
                                                            #Weapon icon Border
                                                            Widget:=texture_block{DefaultImage:=Textures.General.T_GenericBorderSquare, DefaultDesiredSize:=vector2{X:=78.0, Y:=78.0}, DefaultTint:=MakeColorFromHex("#2e7be3FF")}

                                                        overlay_slot:
                                                            VerticalAlignment:=vertical_alignment.Top
                                                            HorizontalAlignment:=horizontal_alignment.Center
                                                            Padding:=margin{Left:=0.0, Top:=5.0, Right:=0.0, Bottom:=0.0}
                                                            #Weapon icon - SMG
                                                            Widget:=texture_block{DefaultImage:=Textures.Items.T_IconSMG, DefaultDesiredSize:=vector2{X:=50.0, Y:=50.0}, DefaultTint:=MakeColorFromHex("#FFFFFFFF")}

                                                        overlay_slot:
                                                            VerticalAlignment:=vertical_alignment.Bottom
                                                            HorizontalAlignment:=horizontal_alignment.Center
                                                            Padding:=margin{Left:=0.0, Top:=0.0, Right:=0.0, Bottom:=5.0}
                                                            #Weapon header - SMG
                                                            Widget:=texture_block{DefaultImage:=Textures.Items.T_HeaderSMGs, DefaultDesiredSize:=vector2{X:=78.0, Y:=18.0}, DefaultTint:=MakeColorFromHex("#FFFFFFFF")}
                                                            

                                            stack_box_slot:
                                                VerticalAlignment:=vertical_alignment.Center
                                                HorizontalAlignment:=horizontal_alignment.Center
                                                Padding:=margin{Left:=0.0, Top:=0.0, Right:=8.0, Bottom:=0.0}
                                                Widget:=overlay:
                                                    Slots:=array:
                                                        #Weapon Icon - Rifle
                                                        overlay_slot:
                                                            VerticalAlignment:=vertical_alignment.Center
                                                            HorizontalAlignment:=horizontal_alignment.Center
                                                            Padding:=margin{Left:=0.0, Top:=0.0, Right:=0.0, Bottom:=0.0}
                                                            #Weapon icon BG
                                                            Widget:=texture_block{DefaultImage:=Textures.General.T_GenericSquare, DefaultDesiredSize:=vector2{X:=78.0, Y:=78.0}, DefaultTint:=MakeColorFromHex("#000c1d88")}

                                                        overlay_slot:
                                                            VerticalAlignment:=vertical_alignment.Top
                                                            HorizontalAlignment:=horizontal_alignment.Center
                                                            Padding:=margin{Left:=0.0, Top:=0.0, Right:=0.0, Bottom:=0.0}
                                                            #Weapon icon Border
                                                            Widget:=texture_block{DefaultImage:=Textures.General.T_GenericBorderSquare, DefaultDesiredSize:=vector2{X:=78.0, Y:=78.0}, DefaultTint:=MakeColorFromHex("#2e7be3FF")}

                                                        overlay_slot:
                                                            VerticalAlignment:=vertical_alignment.Top
                                                            HorizontalAlignment:=horizontal_alignment.Center
                                                            Padding:=margin{Left:=0.0, Top:=5.0, Right:=0.0, Bottom:=0.0}
                                                            #Weapon icon - Rifle
                                                            Widget:=texture_block{DefaultImage:=Textures.Items.T_IconRifle, DefaultDesiredSize:=vector2{X:=50.0, Y:=50.0}, DefaultTint:=MakeColorFromHex("#FFFFFFFF")}

                                                        overlay_slot:
                                                            VerticalAlignment:=vertical_alignment.Bottom
                                                            HorizontalAlignment:=horizontal_alignment.Center
                                                            Padding:=margin{Left:=0.0, Top:=0.0, Right:=0.0, Bottom:=5.0}
                                                            #Weapon header - Rifle
                                                            Widget:=texture_block{DefaultImage:=Textures.Items.T_HeaderRifle, DefaultDesiredSize:=vector2{X:=78.0, Y:=18.0}, DefaultTint:=MakeColorFromHex("#FFFFFFFF")}

                                            stack_box_slot:
                                                VerticalAlignment:=vertical_alignment.Center
                                                HorizontalAlignment:=horizontal_alignment.Center
                                                Padding:=margin{Left:=0.0, Top:=0.0, Right:=8.0, Bottom:=0.0}
                                                Widget:=overlay:
                                                    Slots:=array:
                                                        #Weapon Icon - Explosive
                                                        overlay_slot:
                                                            VerticalAlignment:=vertical_alignment.Center
                                                            HorizontalAlignment:=horizontal_alignment.Center
                                                            Padding:=margin{Left:=0.0, Top:=0.0, Right:=0.0, Bottom:=0.0}
                                                            #Weapon icon BG
                                                            Widget:=texture_block{DefaultImage:=Textures.General.T_GenericSquare, DefaultDesiredSize:=vector2{X:=78.0, Y:=78.0}, DefaultTint:=MakeColorFromHex("#000c1d88")}

                                                        overlay_slot:
                                                            VerticalAlignment:=vertical_alignment.Center
                                                            HorizontalAlignment:=horizontal_alignment.Center
                                                            Padding:=margin{Left:=0.0, Top:=0.0, Right:=0.0, Bottom:=0.0}
                                                            #Weapon icon Border
                                                            Widget:=texture_block{DefaultImage:=Textures.General.T_GenericBorderSquare, DefaultDesiredSize:=vector2{X:=78.0, Y:=78.0}, DefaultTint:=MakeColorFromHex("#2e7be3FF")}

                                                        #overlay_slot:
                                                            #VerticalAlignment:=vertical_alignment.Top
                                                            #HorizontalAlignment:=horizontal_alignment.Center
                                                            #Padding:=margin{Left:=0.0, Top:=5.0, Right:=0.0, Bottom:=0.0}
                                                            #Weapon icon - Explosive
                                                            #Widget:=texture_block{DefaultImage:=Textures.Items.T_IconExplosive, DefaultDesiredSize:=vector2{X:=50.0, Y:=50.0}, DefaultTint:=MakeColorFromHex("#FFFFFFFF")}

                                                        #overlay_slot:
                                                            #VerticalAlignment:=vertical_alignment.Bottom
                                                            #HorizontalAlignment:=horizontal_alignment.Center
                                                            #Padding:=margin{Left:=0.0, Top:=0.0, Right:=0.0, Bottom:=5.0}
                                                            #Weapon header - Explosive
                                                            #Widget:=texture_block{DefaultImage:=Textures.Items.T_HeaderExplosive, DefaultDesiredSize:=vector2{X:=78.0, Y:=18.0}, DefaultTint:=MakeColorFromHex("#FFFFFFFF")}

                                            stack_box_slot:
                                                VerticalAlignment:=vertical_alignment.Center
                                                HorizontalAlignment:=horizontal_alignment.Center
                                                Padding:=margin{Left:=0.0, Top:=0.0, Right:=8.0, Bottom:=0.0}
                                                Widget:=overlay:
                                                    Slots:=array:
                                                        #Weapon Icon - Special
                                                        overlay_slot:
                                                            VerticalAlignment:=vertical_alignment.Center
                                                            HorizontalAlignment:=horizontal_alignment.Center
                                                            Padding:=margin{Left:=0.0, Top:=0.0, Right:=0.0, Bottom:=0.0}
                                                            #Weapon icon BG
                                                            Widget:=texture_block{DefaultImage:=Textures.General.T_GenericSquare, DefaultDesiredSize:=vector2{X:=78.0, Y:=78.0}, DefaultTint:=MakeColorFromHex("#000c1d88")}

                                                        overlay_slot:
                                                            VerticalAlignment:=vertical_alignment.Center
                                                            HorizontalAlignment:=horizontal_alignment.Center
                                                            Padding:=margin{Left:=0.0, Top:=0.0, Right:=0.0, Bottom:=0.0}
                                                            #Weapon icon Border
                                                            Widget:=texture_block{DefaultImage:=Textures.General.T_GenericBorderSquare, DefaultDesiredSize:=vector2{X:=78.0, Y:=78.0}, DefaultTint:=MakeColorFromHex("#2e7be3FF")}

                                                        #overlay_slot:
                                                            #VerticalAlignment:=vertical_alignment.Top
                                                            #HorizontalAlignment:=horizontal_alignment.Center
                                                            #Padding:=margin{Left:=0.0, Top:=5.0, Right:=0.0, Bottom:=0.0}
                                                            #Weapon icon - special
                                                            #Widget:=texture_block{DefaultImage:=Textures.Items.T_IconArtifact, DefaultDesiredSize:=vector2{X:=50.0, Y:=50.0}, DefaultTint:=MakeColorFromHex("#FFFFFFFF")}

                                                        #overlay_slot:
                                                            #VerticalAlignment:=vertical_alignment.Bottom
                                                            #HorizontalAlignment:=horizontal_alignment.Center
                                                            #Padding:=margin{Left:=0.0, Top:=0.0, Right:=0.0, Bottom:=5.0}
                                                            #Weapon header - Special
                                                            #Widget:=texture_block{DefaultImage:=Textures.Items.T_HeaderSpecial, DefaultDesiredSize:=vector2{X:=78.0, Y:=18.0}, DefaultTint:=MakeColorFromHex("#FFFFFFFF")}

                                            stack_box_slot:
                                                VerticalAlignment:=vertical_alignment.Center
                                                HorizontalAlignment:=horizontal_alignment.Center
                                                Padding:=margin{Left:=0.0, Top:=0.0, Right:=8.0, Bottom:=0.0}
                                                Widget:=overlay:
                                                    Slots:=array:
                                                        #Weapon Icon - DMR Sniper
                                                        overlay_slot:
                                                            VerticalAlignment:=vertical_alignment.Center
                                                            HorizontalAlignment:=horizontal_alignment.Center
                                                            Padding:=margin{Left:=0.0, Top:=0.0, Right:=0.0, Bottom:=0.0}
                                                            #Weapon icon BG
                                                            Widget:=texture_block{DefaultImage:=Textures.General.T_GenericSquare, DefaultDesiredSize:=vector2{X:=78.0, Y:=78.0}, DefaultTint:=MakeColorFromHex("#000c1d88")}

                                                        overlay_slot:
                                                            VerticalAlignment:=vertical_alignment.Center
                                                            HorizontalAlignment:=horizontal_alignment.Center
                                                            Padding:=margin{Left:=0.0, Top:=0.0, Right:=0.0, Bottom:=0.0}
                                                            #Weapon icon Border
                                                            Widget:=texture_block{DefaultImage:=Textures.General.T_GenericBorderSquare, DefaultDesiredSize:=vector2{X:=78.0, Y:=78.0}, DefaultTint:=MakeColorFromHex("#2e7be3FF")}

                                                        overlay_slot:
                                                            VerticalAlignment:=vertical_alignment.Top
                                                            HorizontalAlignment:=horizontal_alignment.Center
                                                            Padding:=margin{Left:=0.0, Top:=5.0, Right:=0.0, Bottom:=0.0}
                                                            #Weapon icon - Sniper
                                                            Widget:=texture_block{DefaultImage:=Textures.Items.T_IconSniper, DefaultDesiredSize:=vector2{X:=50.0, Y:=50.0}, DefaultTint:=MakeColorFromHex("#FFFFFFFF")}

                                                        overlay_slot:
                                                            VerticalAlignment:=vertical_alignment.Bottom
                                                            HorizontalAlignment:=horizontal_alignment.Center
                                                            Padding:=margin{Left:=0.0, Top:=0.0, Right:=0.0, Bottom:=5.0}
                                                            #Weapon header - DMR Sniper
                                                            Widget:=texture_block{DefaultImage:=Textures.Items.T_HeaderDMRSniper, DefaultDesiredSize:=vector2{X:=78.0, Y:=18.0}, DefaultTint:=MakeColorFromHex("#FFFFFFFF")}

                                            stack_box_slot:
                                                VerticalAlignment:=vertical_alignment.Fill
                                                HorizontalAlignment:=horizontal_alignment.Center #Center
                                                Padding:=margin{Left:=0.0, Top:=0.0, Right:=0.0, Bottom:=0.0}
                                                Widget:=overlay:
                                                    Slots:=array:
                                                        overlay_slot:
                                                            #Makeshift size box
                                                            VerticalAlignment:=vertical_alignment.Center
                                                            HorizontalAlignment:=horizontal_alignment.Center
                                                            Padding:=margin{Left:=0.0, Top:=0.0, Right:=0.0, Bottom:=0.0}
                                                            Widget:=texture_block{DefaultImage:=Textures.General.T_GenSquTrans, DefaultDesiredSize:=vector2{X:=418.0, Y:=100.0}, DefaultTint:=MakeColorFromHex("#00FF0011")}

                                                        overlay_slot:
                                                            #Vertical stack box to add all the rarity information
                                                            VerticalAlignment:=vertical_alignment.Center
                                                            HorizontalAlignment:=horizontal_alignment.Center
                                                            Padding:=margin{Left:=0.0, Top:=0.0, Right:=0.0, Bottom:=0.0}
                                                            Widget:=stack_box:
                                                                Orientation:=orientation.Vertical
                                                                Slots:=array:
                                                                    stack_box_slot:
                                                                        #Horizontal Stack box for max rarity available
                                                                        VerticalAlignment:=vertical_alignment.Top
                                                                        HorizontalAlignment:=horizontal_alignment.Center
                                                                        Padding:=margin{Left:=0.0, Top:=0.0, Right:=0.0, Bottom:=0.0}
                                                                        Widget:=stack_box:
                                                                            Orientation:=orientation.Horizontal
                                                                            Slots:=array:
                                                                                stack_box_slot:
                                                                                    VerticalAlignment:=vertical_alignment.Center
                                                                                    HorizontalAlignment:=horizontal_alignment.Center
                                                                                    Padding:=margin{Left:=0.0, Top:=0.0, Right:=0.0, Bottom:=0.0}
                                                                                    Widget:=text_block{DefaultText := StringToMessage("Rarity chance up to "), DefaultTextColor:=MakeColorFromHex("#FFFFFFFF")}

                                                                                stack_box_slot:
                                                                                    VerticalAlignment:=vertical_alignment.Center
                                                                                    HorizontalAlignment:=horizontal_alignment.Center
                                                                                    Padding:=margin{Left:=0.0, Top:=0.0, Right:=0.0, Bottom:=0.0}
                                                                                    Widget:=texture_block{DefaultImage:=Textures.Rarity.T_FlagUncommon, DefaultDesiredSize:=vector2{X:=104.0, Y:=36.0}, DefaultTint:=MakeColorFromHex("#FFFFFFFF")}

                                                                    stack_box_slot:
                                                                        #Horizontal overlay for overall rarity scheme
                                                                        VerticalAlignment:=vertical_alignment.Bottom
                                                                        HorizontalAlignment:=horizontal_alignment.Center
                                                                        Padding:=margin{Left:=0.0, Top:=0.0, Right:=0.0, Bottom:=0.0}
                                                                        Widget:=overlay:
                                                                            Slots:=array:
                                                                                
                                                                                #Makeshift size box for the rarity scheme
                                                                                overlay_slot:
                                                                                    VerticalAlignment:=vertical_alignment.Center
                                                                                    HorizontalAlignment:=horizontal_alignment.Center
                                                                                    Padding:=margin{Left:=0.0, Top:=0.0, Right:=0.0, Bottom:=0.0}
                                                                                    Widget:=texture_block{DefaultImage:=Textures.General.T_GenSquTrans, DefaultDesiredSize:=vector2{X:=255.0, Y:=50.0}, DefaultTint:=MakeColorFromHex("#FF000022")}#Red frame

                                                                                #BG color for rarity section
                                                                                overlay_slot:
                                                                                    VerticalAlignment:=vertical_alignment.Center
                                                                                    HorizontalAlignment:=horizontal_alignment.Center
                                                                                    Padding:=margin{Left:=0.0, Top:=0.0, Right:=0.0, Bottom:=0.0}
                                                                                    Widget:=texture_block{DefaultImage:=Textures.General.T_GenericSquare, DefaultDesiredSize:=vector2{X:=255.0, Y:=40.0}, DefaultTint:=MakeColorFromHex("#07182FFF")}

                                                                                #Visible border for rarity layout
                                                                                overlay_slot:
                                                                                    VerticalAlignment:=vertical_alignment.Center
                                                                                    HorizontalAlignment:=horizontal_alignment.Center
                                                                                    Padding:=margin{Left:=0.0, Top:=0.0, Right:=0.0, Bottom:=0.0}
                                                                                    Widget:=texture_block{DefaultImage:=Textures.General.T_GenericBorderSquare3, DefaultDesiredSize:=vector2{X:=255.0, Y:=40.0}, DefaultTint:=MakeColorFromHex("#2e7be3FF")}

                                                                                #Horizontal Stack box for rarity slots
                                                                                overlay_slot:
                                                                                    VerticalAlignment:=vertical_alignment.Fill
                                                                                    HorizontalAlignment:=horizontal_alignment.Fill
                                                                                    Padding:=margin{Left:=5.0, Top:=0.0, Right:=5.0, Bottom:=0.0}
                                                                                    Widget:=stack_box:
                                                                                        Orientation:=orientation.Horizontal
                                                                                        Slots:=array:
                                                                                            #Overlay for each rarity slots, including the pointing arrow
                                                                                            stack_box_slot:
                                                                                                VerticalAlignment:=vertical_alignment.Center
                                                                                                HorizontalAlignment:=horizontal_alignment.Center
                                                                                                Distribution:=option{1.0}
                                                                                                Padding:=margin{Left:=0.0, Top:=0.0, Right:=0.0, Bottom:=0.0}
                                                                                                Widget:=overlay:
                                                                                                    Slots:=array:
                                                                                                        #Makeshift size box
                                                                                                        overlay_slot:
                                                                                                            VerticalAlignment:=vertical_alignment.Center
                                                                                                            HorizontalAlignment:=horizontal_alignment.Center
                                                                                                            Padding:=margin{Left:=0.0, Top:=0.0, Right:=0.0, Bottom:=0.0}
                                                                                                            Widget:=texture_block{DefaultImage:=Textures.General.T_GenSquTrans, DefaultDesiredSize:=vector2{X:=30.0, Y:=40.0}, DefaultTint:=MakeColorFromHex("#0000FF08")}

                                                                                                        #Overlay for rarity border and texture
                                                                                                        overlay_slot:
                                                                                                            VerticalAlignment:=vertical_alignment.Center
                                                                                                            HorizontalAlignment:=horizontal_alignment.Center
                                                                                                            Padding:=margin{Left:=0.0, Top:=0.0, Right:=0.0, Bottom:=0.0}
                                                                                                            Widget:=overlay:
                                                                                                                Slots:=array:
                                                                                                                    
                                                                                                                    #Rarity texture border, when the rarity slot is empty
                                                                                                                    #overlay_slot:
                                                                                                                        #VerticalAlignment:=vertical_alignment.Center
                                                                                                                        #HorizontalAlignment:=horizontal_alignment.Center
                                                                                                                        #Padding:=margin{Left:=0.0, Top:=0.0, Right:=0.0, Bottom:=0.0}
                                                                                                                        #Widget:=texture_block{DefaultImage:=Textures.General.T_GenericBorderSquare2, DefaultDesiredSize:=vector2{X:=30.0, Y:=30.0}, DefaultTint:=MakeColorFromHex("#2e7be388")}

                                                                                                                    #Rarity texture
                                                                                                                    overlay_slot:
                                                                                                                        VerticalAlignment:=vertical_alignment.Center
                                                                                                                        HorizontalAlignment:=horizontal_alignment.Center
                                                                                                                        Padding:=margin{Left:=0.0, Top:=0.0, Right:=0.0, Bottom:=0.0}
                                                                                                                        Widget:=texture_block{DefaultImage:=Textures.General.T_GenericSquare, DefaultDesiredSize:=vector2{X:=20.0, Y:=20.0}, DefaultTint:=MakeColorFromHex("#979797FF")}
                                                                                                        
                                                                                                                    

                                                                                                        #Bracket
                                                                                                        overlay_slot:
                                                                                                            VerticalAlignment:=vertical_alignment.Fill
                                                                                                            HorizontalAlignment:=horizontal_alignment.Fill
                                                                                                            Padding:=margin{Left:=0.0, Top:=0.0, Right:=0.0, Bottom:=0.0}
                                                                                                            Widget:=stack_box:
                                                                                                                Orientation:=orientation.Horizontal
                                                                                                                #two instances of bracket texture to represent the range of rarities
                                                                                                                Slots:=array:

                                                                                                                    stack_box_slot:
                                                                                                                        VerticalAlignment:=vertical_alignment.Center
                                                                                                                        HorizontalAlignment:=horizontal_alignment.Left
                                                                                                                        Distribution:=option{1.0}
                                                                                                                        Padding:=margin{Left:=0.0, Top:=0.0, Right:=0.0, Bottom:=0.0}
                                                                                                                        Widget:=texture_block{DefaultImage:=Textures.General.T_RarityBracketLeft, DefaultDesiredSize:=vector2{X:=18.0, Y:=36.0}, DefaultTint:=MakeColorFromHex("#FFFFFF44")}

                                                                                                                    #stack_box_slot:
                                                                                                                        #VerticalAlignment:=vertical_alignment.Center
                                                                                                                        #HorizontalAlignment:=horizontal_alignment.Right
                                                                                                                        #Distribution:=option{1.0}
                                                                                                                        #Padding:=margin{Left:=0.0, Top:=0.0, Right:=0.0, Bottom:=0.0}
                                                                                                                        #Widget:=texture_block{DefaultImage:=Textures.General.T_RarityBracketRight, DefaultDesiredSize:=vector2{X:=18.0, Y:=36.0}, DefaultTint:=MakeColorFromHex("#FFFFFF44")}

                                                                                            stack_box_slot:
                                                                                                VerticalAlignment:=vertical_alignment.Center
                                                                                                HorizontalAlignment:=horizontal_alignment.Center
                                                                                                Distribution:=option{1.0}
                                                                                                Padding:=margin{Left:=0.0, Top:=0.0, Right:=0.0, Bottom:=0.0}
                                                                                                Widget:=overlay:
                                                                                                    Slots:=array:
                                                                                                        #Makeshift size box
                                                                                                        overlay_slot:
                                                                                                            VerticalAlignment:=vertical_alignment.Center
                                                                                                            HorizontalAlignment:=horizontal_alignment.Center
                                                                                                            Padding:=margin{Left:=0.0, Top:=0.0, Right:=0.0, Bottom:=0.0}
                                                                                                            Widget:=texture_block{DefaultImage:=Textures.General.T_GenSquTrans, DefaultDesiredSize:=vector2{X:=30.0, Y:=40.0}, DefaultTint:=MakeColorFromHex("#0000FF08")}

                                                                                                        #Overlay for rarity border and texture
                                                                                                        overlay_slot:
                                                                                                            VerticalAlignment:=vertical_alignment.Center
                                                                                                            HorizontalAlignment:=horizontal_alignment.Center
                                                                                                            Padding:=margin{Left:=0.0, Top:=0.0, Right:=0.0, Bottom:=0.0}
                                                                                                            Widget:=overlay:
                                                                                                                Slots:=array:
                                                                                                                    
                                                                                                                    #Rarity texture border, when the rarity slot is empty
                                                                                                                    #overlay_slot:
                                                                                                                        #VerticalAlignment:=vertical_alignment.Center
                                                                                                                        #HorizontalAlignment:=horizontal_alignment.Center
                                                                                                                        #Padding:=margin{Left:=0.0, Top:=0.0, Right:=0.0, Bottom:=0.0}
                                                                                                                        #Widget:=texture_block{DefaultImage:=Textures.General.T_GenericBorderSquare2, DefaultDesiredSize:=vector2{X:=30.0, Y:=30.0}, DefaultTint:=MakeColorFromHex("#2e7be388")}

                                                                                                                    #Rarity texture
                                                                                                                    overlay_slot:
                                                                                                                        VerticalAlignment:=vertical_alignment.Center
                                                                                                                        HorizontalAlignment:=horizontal_alignment.Center
                                                                                                                        Padding:=margin{Left:=0.0, Top:=0.0, Right:=0.0, Bottom:=0.0}
                                                                                                                        Widget:=texture_block{DefaultImage:=Textures.General.T_GenericSquare, DefaultDesiredSize:=vector2{X:=20.0, Y:=20.0}, DefaultTint:=MakeColorFromHex("#4ccc2aFF")}

                                                                                                        #Bracket
                                                                                                        overlay_slot:
                                                                                                            VerticalAlignment:=vertical_alignment.Fill
                                                                                                            HorizontalAlignment:=horizontal_alignment.Fill
                                                                                                            Padding:=margin{Left:=0.0, Top:=0.0, Right:=0.0, Bottom:=0.0}
                                                                                                            Widget:=stack_box:
                                                                                                                Orientation:=orientation.Horizontal
                                                                                                                #two instances of bracket texture to represent the range of rarities
                                                                                                                Slots:=array:

                                                                                                                    #stack_box_slot:
                                                                                                                        #VerticalAlignment:=vertical_alignment.Center
                                                                                                                        #HorizontalAlignment:=horizontal_alignment.Left
                                                                                                                        #Distribution:=option{1.0}
                                                                                                                        #Padding:=margin{Left:=0.0, Top:=0.0, Right:=0.0, Bottom:=0.0}
                                                                                                                        #Widget:=texture_block{DefaultImage:=Textures.General.T_RarityBracketLeft, DefaultDesiredSize:=vector2{X:=18.0, Y:=36.0}, DefaultTint:=MakeColorFromHex("#FFFFFF44")}

                                                                                                                    stack_box_slot:
                                                                                                                        VerticalAlignment:=vertical_alignment.Center
                                                                                                                        HorizontalAlignment:=horizontal_alignment.Right
                                                                                                                        Distribution:=option{1.0}
                                                                                                                        Padding:=margin{Left:=0.0, Top:=0.0, Right:=0.0, Bottom:=0.0}
                                                                                                                        Widget:=texture_block{DefaultImage:=Textures.General.T_RarityBracketRight, DefaultDesiredSize:=vector2{X:=18.0, Y:=36.0}, DefaultTint:=MakeColorFromHex("#FFFFFF44")}

                                                                                            stack_box_slot:
                                                                                                VerticalAlignment:=vertical_alignment.Center
                                                                                                HorizontalAlignment:=horizontal_alignment.Center
                                                                                                Distribution:=option{1.0}
                                                                                                Padding:=margin{Left:=0.0, Top:=0.0, Right:=0.0, Bottom:=0.0}
                                                                                                Widget:=overlay:
                                                                                                    Slots:=array:
                                                                                                        #Makeshift size box
                                                                                                        overlay_slot:
                                                                                                            VerticalAlignment:=vertical_alignment.Center
                                                                                                            HorizontalAlignment:=horizontal_alignment.Center
                                                                                                            Padding:=margin{Left:=0.0, Top:=0.0, Right:=0.0, Bottom:=0.0}
                                                                                                            Widget:=texture_block{DefaultImage:=Textures.General.T_GenSquTrans, DefaultDesiredSize:=vector2{X:=30.0, Y:=40.0}, DefaultTint:=MakeColorFromHex("#0000FF08")}

                                                                                                        #Overlay for rarity border and texture
                                                                                                        overlay_slot:
                                                                                                            VerticalAlignment:=vertical_alignment.Center
                                                                                                            HorizontalAlignment:=horizontal_alignment.Center
                                                                                                            Padding:=margin{Left:=0.0, Top:=0.0, Right:=0.0, Bottom:=0.0}
                                                                                                            Widget:=overlay:
                                                                                                                Slots:=array:
                                                                                                                                                                                                                
                                                                                                                    #Rarity texture border, when the rarity slot is empty
                                                                                                                    overlay_slot:
                                                                                                                        VerticalAlignment:=vertical_alignment.Center
                                                                                                                        HorizontalAlignment:=horizontal_alignment.Center
                                                                                                                        Padding:=margin{Left:=0.0, Top:=0.0, Right:=0.0, Bottom:=0.0}
                                                                                                                        Widget:=texture_block{DefaultImage:=Textures.General.T_GenericBorderSquare2, DefaultDesiredSize:=vector2{X:=30.0, Y:=30.0}, DefaultTint:=MakeColorFromHex("#2e7be388")}
                                                                                                                                                                                                                            
                                                                                                                    #Rarity texture
                                                                                                                    #overlay_slot:
                                                                                                                        #VerticalAlignment:=vertical_alignment.Center
                                                                                                                        #HorizontalAlignment:=horizontal_alignment.Center
                                                                                                                        #Padding:=margin{Left:=0.0, Top:=0.0, Right:=0.0, Bottom:=0.0}
                                                                                                                        #Widget:=texture_block{DefaultImage:=Textures.General.T_GenericSquare, DefaultDesiredSize:=vector2{X:=30.0, Y:=30.0}, DefaultTint:=MakeColorFromHex("#2a54ccFF")}
                                                                                                                    
                                                                                                        #Bracket
                                                                                                        overlay_slot:
                                                                                                            VerticalAlignment:=vertical_alignment.Fill
                                                                                                            HorizontalAlignment:=horizontal_alignment.Fill
                                                                                                            Padding:=margin{Left:=0.0, Top:=0.0, Right:=0.0, Bottom:=0.0}
                                                                                                            Widget:=stack_box:
                                                                                                                Orientation:=orientation.Horizontal
                                                                                                                #two instances of bracket texture to represent the range of rarities
                                                                                                                Slots:=array:

                                                                                                                    #stack_box_slot:
                                                                                                                        #VerticalAlignment:=vertical_alignment.Center
                                                                                                                        #HorizontalAlignment:=horizontal_alignment.Left
                                                                                                                        #Distribution:=option{1.0}
                                                                                                                        #Padding:=margin{Left:=0.0, Top:=0.0, Right:=0.0, Bottom:=0.0}
                                                                                                                        #Widget:=texture_block{DefaultImage:=Textures.General.T_RarityBracketLeft, DefaultDesiredSize:=vector2{X:=18.0, Y:=36.0}, DefaultTint:=MakeColorFromHex("#FFFFFF44")}

                                                                                                                    #stack_box_slot:
                                                                                                                        #VerticalAlignment:=vertical_alignment.Center
                                                                                                                        #HorizontalAlignment:=horizontal_alignment.Right
                                                                                                                        #Distribution:=option{1.0}
                                                                                                                        #Padding:=margin{Left:=0.0, Top:=0.0, Right:=0.0, Bottom:=0.0}
                                                                                                                        #Widget:=texture_block{DefaultImage:=Textures.General.T_RarityBracketRight, DefaultDesiredSize:=vector2{X:=18.0, Y:=36.0}, DefaultTint:=MakeColorFromHex("#FFFFFF44")}

                                                                                            stack_box_slot:
                                                                                                VerticalAlignment:=vertical_alignment.Center
                                                                                                HorizontalAlignment:=horizontal_alignment.Center
                                                                                                Distribution:=option{1.0}
                                                                                                Padding:=margin{Left:=0.0, Top:=0.0, Right:=0.0, Bottom:=0.0}
                                                                                                Widget:=overlay:
                                                                                                    Slots:=array:
                                                                                                        #Makeshift size box
                                                                                                        overlay_slot:
                                                                                                            VerticalAlignment:=vertical_alignment.Center
                                                                                                            HorizontalAlignment:=horizontal_alignment.Center
                                                                                                            Padding:=margin{Left:=0.0, Top:=0.0, Right:=0.0, Bottom:=0.0}
                                                                                                            Widget:=texture_block{DefaultImage:=Textures.General.T_GenSquTrans, DefaultDesiredSize:=vector2{X:=30.0, Y:=40.0}, DefaultTint:=MakeColorFromHex("#0000FF08")}

                                                                                                        #Overlay for rarity border and texture
                                                                                                        overlay_slot:
                                                                                                            VerticalAlignment:=vertical_alignment.Center
                                                                                                            HorizontalAlignment:=horizontal_alignment.Center
                                                                                                            Padding:=margin{Left:=0.0, Top:=0.0, Right:=0.0, Bottom:=0.0}
                                                                                                            Widget:=overlay:
                                                                                                                Slots:=array:

                                                                                                                    #Rarity texture border, when the rarity slot is empty
                                                                                                                    overlay_slot:
                                                                                                                        VerticalAlignment:=vertical_alignment.Center
                                                                                                                        HorizontalAlignment:=horizontal_alignment.Center
                                                                                                                        Padding:=margin{Left:=0.0, Top:=0.0, Right:=0.0, Bottom:=0.0}
                                                                                                                        Widget:=texture_block{DefaultImage:=Textures.General.T_GenericBorderSquare2, DefaultDesiredSize:=vector2{X:=30.0, Y:=30.0}, DefaultTint:=MakeColorFromHex("#2e7be388")}
                                                                                                                    
                                                                                                                    #Rarity texture
                                                                                                                    #overlay_slot:
                                                                                                                        #VerticalAlignment:=vertical_alignment.Center
                                                                                                                        #HorizontalAlignment:=horizontal_alignment.Center
                                                                                                                        #Padding:=margin{Left:=0.0, Top:=0.0, Right:=0.0, Bottom:=0.0}
                                                                                                                        #Widget:=texture_block{DefaultImage:=Textures.General.T_GenericSquare, DefaultDesiredSize:=vector2{X:=30.0, Y:=30.0}, DefaultTint:=MakeColorFromHex("#b32accFF")}
                                                                                                                    
                                                                                                        #Bracket
                                                                                                        overlay_slot:
                                                                                                            VerticalAlignment:=vertical_alignment.Fill
                                                                                                            HorizontalAlignment:=horizontal_alignment.Fill
                                                                                                            Padding:=margin{Left:=0.0, Top:=0.0, Right:=0.0, Bottom:=0.0}
                                                                                                            Widget:=stack_box:
                                                                                                                Orientation:=orientation.Horizontal
                                                                                                                #two instances of bracket texture to represent the range of rarities
                                                                                                                Slots:=array:

                                                                                                                    #stack_box_slot:
                                                                                                                        #VerticalAlignment:=vertical_alignment.Center
                                                                                                                        #HorizontalAlignment:=horizontal_alignment.Left
                                                                                                                        #Distribution:=option{1.0}
                                                                                                                        #Padding:=margin{Left:=0.0, Top:=0.0, Right:=0.0, Bottom:=0.0}
                                                                                                                        #Widget:=texture_block{DefaultImage:=Textures.General.T_RarityBracketLeft, DefaultDesiredSize:=vector2{X:=18.0, Y:=36.0}, DefaultTint:=MakeColorFromHex("#FFFFFF44")}

                                                                                                                    #stack_box_slot:
                                                                                                                        #VerticalAlignment:=vertical_alignment.Center
                                                                                                                        #HorizontalAlignment:=horizontal_alignment.Right
                                                                                                                        #Distribution:=option{1.0}
                                                                                                                        #Padding:=margin{Left:=0.0, Top:=0.0, Right:=0.0, Bottom:=0.0}
                                                                                                                        #Widget:=texture_block{DefaultImage:=Textures.General.T_RarityBracketRight, DefaultDesiredSize:=vector2{X:=18.0, Y:=36.0}, DefaultTint:=MakeColorFromHex("#FFFFFF44")}

                                                                                            stack_box_slot:
                                                                                                VerticalAlignment:=vertical_alignment.Center
                                                                                                HorizontalAlignment:=horizontal_alignment.Center
                                                                                                Distribution:=option{1.0}
                                                                                                Padding:=margin{Left:=0.0, Top:=0.0, Right:=0.0, Bottom:=0.0}
                                                                                                Widget:=overlay:
                                                                                                    Slots:=array:
                                                                                                        #Makeshift size box
                                                                                                        overlay_slot:
                                                                                                            VerticalAlignment:=vertical_alignment.Center
                                                                                                            HorizontalAlignment:=horizontal_alignment.Center
                                                                                                            Padding:=margin{Left:=0.0, Top:=0.0, Right:=0.0, Bottom:=0.0}
                                                                                                            Widget:=texture_block{DefaultImage:=Textures.General.T_GenSquTrans, DefaultDesiredSize:=vector2{X:=30.0, Y:=40.0}, DefaultTint:=MakeColorFromHex("#0000FF08")}

                                                                                                        #Overlay for rarity border and texture
                                                                                                        overlay_slot:
                                                                                                            VerticalAlignment:=vertical_alignment.Center
                                                                                                            HorizontalAlignment:=horizontal_alignment.Center
                                                                                                            Padding:=margin{Left:=0.0, Top:=0.0, Right:=0.0, Bottom:=0.0}
                                                                                                            Widget:=overlay:
                                                                                                                Slots:=array:

                                                                                                                    #Rarity texture border, when the rarity slot is empty
                                                                                                                    overlay_slot:
                                                                                                                        VerticalAlignment:=vertical_alignment.Center
                                                                                                                        HorizontalAlignment:=horizontal_alignment.Center
                                                                                                                        Padding:=margin{Left:=0.0, Top:=0.0, Right:=0.0, Bottom:=0.0}
                                                                                                                        Widget:=texture_block{DefaultImage:=Textures.General.T_GenericBorderSquare2, DefaultDesiredSize:=vector2{X:=30.0, Y:=30.0}, DefaultTint:=MakeColorFromHex("#2e7be388")}
                                                                                                                    
                                                                                                                    #Rarity texture
                                                                                                                    #overlay_slot:
                                                                                                                        #VerticalAlignment:=vertical_alignment.Center
                                                                                                                        #HorizontalAlignment:=horizontal_alignment.Center
                                                                                                                        #Padding:=margin{Left:=0.0, Top:=0.0, Right:=0.0, Bottom:=0.0}
                                                                                                                        #Widget:=texture_block{DefaultImage:=Textures.General.T_GenericSquare, DefaultDesiredSize:=vector2{X:=30.0, Y:=30.0}, DefaultTint:=MakeColorFromHex("#cc8d2aFF")}

                                                                                                        #Bracket
                                                                                                        overlay_slot:
                                                                                                            VerticalAlignment:=vertical_alignment.Fill
                                                                                                            HorizontalAlignment:=horizontal_alignment.Fill
                                                                                                            Padding:=margin{Left:=0.0, Top:=0.0, Right:=0.0, Bottom:=0.0}
                                                                                                            Widget:=stack_box:
                                                                                                                Orientation:=orientation.Horizontal
                                                                                                                #two instances of bracket texture to represent the range of rarities
                                                                                                                Slots:=array:

                                                                                                                    #stack_box_slot:
                                                                                                                        #VerticalAlignment:=vertical_alignment.Center
                                                                                                                        #HorizontalAlignment:=horizontal_alignment.Left
                                                                                                                        #Distribution:=option{1.0}
                                                                                                                        #Padding:=margin{Left:=0.0, Top:=0.0, Right:=0.0, Bottom:=0.0}
                                                                                                                        #Widget:=texture_block{DefaultImage:=Textures.General.T_RarityBracketLeft, DefaultDesiredSize:=vector2{X:=18.0, Y:=36.0}, DefaultTint:=MakeColorFromHex("#FFFFFF44")}

                                                                                                                    #stack_box_slot:
                                                                                                                        #VerticalAlignment:=vertical_alignment.Center
                                                                                                                        #HorizontalAlignment:=horizontal_alignment.Right
                                                                                                                        #Distribution:=option{1.0}
                                                                                                                        #Padding:=margin{Left:=0.0, Top:=0.0, Right:=0.0, Bottom:=0.0}
                                                                                                                        #Widget:=texture_block{DefaultImage:=Textures.General.T_RarityBracketRight, DefaultDesiredSize:=vector2{X:=18.0, Y:=36.0}, DefaultTint:=MakeColorFromHex("#FFFFFF44")}

                                                                                            stack_box_slot:
                                                                                                VerticalAlignment:=vertical_alignment.Center
                                                                                                HorizontalAlignment:=horizontal_alignment.Center
                                                                                                Distribution:=option{1.0}
                                                                                                Padding:=margin{Left:=0.0, Top:=0.0, Right:=0.0, Bottom:=0.0}
                                                                                                Widget:=overlay:
                                                                                                    Slots:=array:
                                                                                                        #Makeshift size box
                                                                                                        overlay_slot:
                                                                                                            VerticalAlignment:=vertical_alignment.Center
                                                                                                            HorizontalAlignment:=horizontal_alignment.Center
                                                                                                            Padding:=margin{Left:=0.0, Top:=0.0, Right:=0.0, Bottom:=0.0}
                                                                                                            Widget:=texture_block{DefaultImage:=Textures.General.T_GenSquTrans, DefaultDesiredSize:=vector2{X:=30.0, Y:=40.0}, DefaultTint:=MakeColorFromHex("#0000FF08")}

                                                                                                        #Overlay for rarity border and texture
                                                                                                        overlay_slot:
                                                                                                            VerticalAlignment:=vertical_alignment.Center
                                                                                                            HorizontalAlignment:=horizontal_alignment.Center
                                                                                                            Padding:=margin{Left:=0.0, Top:=0.0, Right:=0.0, Bottom:=0.0}
                                                                                                            Widget:=overlay:
                                                                                                                Slots:=array:
                                                                                                        
                                                                                                                    #Rarity texture border, when the rarity slot is empty
                                                                                                                    overlay_slot:
                                                                                                                        VerticalAlignment:=vertical_alignment.Center
                                                                                                                        HorizontalAlignment:=horizontal_alignment.Center
                                                                                                                        Padding:=margin{Left:=0.0, Top:=0.0, Right:=0.0, Bottom:=0.0}
                                                                                                                        Widget:=texture_block{DefaultImage:=Textures.General.T_GenericBorderSquare2, DefaultDesiredSize:=vector2{X:=30.0, Y:=30.0}, DefaultTint:=MakeColorFromHex("#2e7be388")}
                                                                                                                    
                                                                                                                    #Rarity texture
                                                                                                                    #overlay_slot:
                                                                                                                        #VerticalAlignment:=vertical_alignment.Center
                                                                                                                        #HorizontalAlignment:=horizontal_alignment.Center
                                                                                                                        #Padding:=margin{Left:=0.0, Top:=0.0, Right:=0.0, Bottom:=0.0}
                                                                                                                        #Widget:=texture_block{DefaultImage:=Textures.General.T_GenericSquare, DefaultDesiredSize:=vector2{X:=30.0, Y:=30.0}, DefaultTint:=MakeColorFromHex("#eacf19FF")}
                                                                                                                    
                                                                                                        #Bracket
                                                                                                        overlay_slot:
                                                                                                            VerticalAlignment:=vertical_alignment.Fill
                                                                                                            HorizontalAlignment:=horizontal_alignment.Fill
                                                                                                            Padding:=margin{Left:=0.0, Top:=0.0, Right:=0.0, Bottom:=0.0}
                                                                                                            Widget:=stack_box:
                                                                                                                Orientation:=orientation.Horizontal
                                                                                                                #two instances of bracket texture to represent the range of rarities
                                                                                                                Slots:=array:

                                                                                                                    #stack_box_slot:
                                                                                                                        #VerticalAlignment:=vertical_alignment.Center
                                                                                                                        #HorizontalAlignment:=horizontal_alignment.Left
                                                                                                                        #Distribution:=option{1.0}
                                                                                                                        #Padding:=margin{Left:=0.0, Top:=0.0, Right:=0.0, Bottom:=0.0}
                                                                                                                        #Widget:=texture_block{DefaultImage:=Textures.General.T_RarityBracketLeft, DefaultDesiredSize:=vector2{X:=18.0, Y:=36.0}, DefaultTint:=MakeColorFromHex("#FFFFFF44")}

                                                                                                                    #stack_box_slot:
                                                                                                                        #VerticalAlignment:=vertical_alignment.Center
                                                                                                                        #HorizontalAlignment:=horizontal_alignment.Right
                                                                                                                        #Distribution:=option{1.0}
                                                                                                                        #Padding:=margin{Left:=0.0, Top:=0.0, Right:=0.0, Bottom:=0.0}
                                                                                                                        #Widget:=texture_block{DefaultImage:=Textures.General.T_RarityBracketRight, DefaultDesiredSize:=vector2{X:=18.0, Y:=36.0}, DefaultTint:=MakeColorFromHex("#FFFFFF44")}

                    #Highlight texture
                    #overlay_slot:
                        #VerticalAlignment:=vertical_alignment.Fill
                        #HorizontalAlignment:=horizontal_alignment.Fill
                        #Padding:=margin{Left:=0.0, Top:=0.0, Right:=0.0, Bottom:=0.0}
                        #Widget:=texture_block{DefaultImage:=Textures.General.T_LevelPointer, DefaultDesiredSize:=vector2{X:=20.0, Y:=20.0}, DefaultTint:=MakeColorFromHex("#7de2FFFF")}

            #Definition for the level row 2
            LevelRow2:=overlay:
                Slots:=array:
                    overlay_slot:
                        #Makeshift size box
                        VerticalAlignment:=vertical_alignment.Center
                        HorizontalAlignment:=horizontal_alignment.Center
                        Padding:=margin{Left:=0.0, Top:=0.0, Right:=0.0, Bottom:=0.0}
                        Widget:=texture_block{DefaultImage:=Textures.General.T_GenSquTrans, DefaultDesiredSize:=vector2{X:=1250.0, Y:=114.0}, DefaultTint:=MakeColorFromHex("#0000FFFF")}

                    overlay_slot:
                        #Rarity level content
                        VerticalAlignment:=vertical_alignment.Center
                        HorizontalAlignment:=horizontal_alignment.Left
                        Padding:=margin{Left:=0.0, Top:=0.0, Right:=0.0, Bottom:=0.0}
                        Widget:=overlay:
                            Slots:=array:

                                overlay_slot:
                                    #Makeshift size box
                                    VerticalAlignment:=vertical_alignment.Center
                                    HorizontalAlignment:=horizontal_alignment.Center
                                    Padding:=margin{Left:=0.0, Top:=0.0, Right:=0.0, Bottom:=0.0}
                                    Widget:=texture_block{DefaultImage:=Textures.General.T_GenericSquare, DefaultDesiredSize:=vector2{X:=1190.0, Y:=114.0}, DefaultTint:=MakeColorFromHex("#010a1832")}

                                overlay_slot:
                                    #Highlight texture, over the content but under the rarity flag
                                    VerticalAlignment:=vertical_alignment.Center
                                    HorizontalAlignment:=horizontal_alignment.Center
                                    Padding:=margin{Left:=0.0, Top:=0.0, Right:=0.0, Bottom:=0.0}
                                    Widget:=texture_block{DefaultImage:=Textures.General.T_LevelBorder, DefaultDesiredSize:=vector2{X:=1190.0, Y:=114.0}, DefaultTint:=MakeColorFromHex("#2e7be3ff")}

                                    #Horizontal stack box
                                overlay_slot:
                                    VerticalAlignment:=vertical_alignment.Fill
                                    HorizontalAlignment:=horizontal_alignment.Left
                                    Padding:=margin{Left:=0.0, Top:=0.0, Right:=0.0, Bottom:=0.0}
                                    Widget:=stack_box:
                                        Orientation:=orientation.Horizontal
                                        Slots:=array:
                                            #Level flag
                                            stack_box_slot:
                                                VerticalAlignment:=vertical_alignment.Top
                                                HorizontalAlignment:=horizontal_alignment.Left
                                                Padding:=margin{Left:=0.0, Top:=0.0, Right:=15.0, Bottom:=0.0}
                                                Widget:=overlay:

                                                    Slots:=array:
                                                        overlay_slot:
                                                            VerticalAlignment:=vertical_alignment.Center
                                                            HorizontalAlignment:=horizontal_alignment.Center
                                                            Padding:=margin{Left:=0.0, Top:=0.0, Right:=0.0, Bottom:=0.0}
                                                            #Tier level makeshift size box
                                                            Widget:=texture_block{DefaultImage:=Textures.General.T_GenSquTrans, DefaultDesiredSize:=vector2{X:=150.0, Y:=114.0}, DefaultTint:=MakeColorFromHex("#FF0000FF")}

                                                        overlay_slot:
                                                            VerticalAlignment:=vertical_alignment.Top
                                                            HorizontalAlignment:=horizontal_alignment.Left
                                                            Padding:=margin{Left:=0.0, Top:=0.0, Right:=0.0, Bottom:=0.0}
                                                            #Selected level yellow flag
                                                            Widget:=overlay:

                                                                Slots:=array:
                                                                    #overlay_slot:
                                                                        #VerticalAlignment:=vertical_alignment.Center
                                                                        #HorizontalAlignment:=horizontal_alignment.Center
                                                                        #Padding:=margin{Left:=0.0, Top:=0.0, Right:=0.0, Bottom:=0.0}
                                                                        #level flag texture
                                                                        #Widget:=texture_block{DefaultImage:=Textures.General.T_GenLevelFlag, DefaultDesiredSize:=vector2{X:=140.0, Y:=45.0}, DefaultTint:=MakeColorFromHex("#f7ff1aFF")}

                                                                    #overlay_slot:
                                                                        #VerticalAlignment:=vertical_alignment.Center
                                                                        #HorizontalAlignment:=horizontal_alignment.Center
                                                                        #Padding:=margin{Left:=0.0, Top:=0.0, Right:=0.0, Bottom:=0.0}
                                                                        #level flag text
                                                                        #Widget:=text_block{DefaultText := StringToMessage("Level 0"), DefaultTextColor:=MakeColorFromHex("#000000FF")}

                                                        overlay_slot:
                                                            VerticalAlignment:=vertical_alignment.Fill
                                                            HorizontalAlignment:=horizontal_alignment.Center
                                                            Padding:=margin{Left:=0.0, Top:=5.0, Right:=0.0, Bottom:=5.0}
                                                            #Unselected level and lock
                                                            Widget:=overlay:
                                                                Slots:=array:
                                                                    
                                                                    overlay_slot:
                                                                        VerticalAlignment:=vertical_alignment.Top
                                                                        HorizontalAlignment:=horizontal_alignment.Center
                                                                        Padding:=margin{Left:=0.0, Top:=0.0, Right:=0.0, Bottom:=0.0}
                                                                        #level flag texture
                                                                        Widget:=text_block{DefaultText := StringToMessage("Level 2"), DefaultTextColor:=MakeColorFromHex("#FFFFFFFF")}

                                                                    overlay_slot:
                                                                        VerticalAlignment:=vertical_alignment.Bottom
                                                                        HorizontalAlignment:=horizontal_alignment.Center
                                                                        Padding:=margin{Left:=0.0, Top:=0.0, Right:=0.0, Bottom:=0.0}
                                                                        #level flag texture
                                                                        Widget:=texture_block{DefaultImage:=Textures.General.T_Lock, DefaultDesiredSize:=vector2{X:=52.0, Y:=58.0}, DefaultTint:=MakeColorFromHex("#ffffff88")}
                                            
                                            #Level weapon roster
                                            stack_box_slot:
                                                VerticalAlignment:=vertical_alignment.Center
                                                HorizontalAlignment:=horizontal_alignment.Center
                                                Padding:=margin{Left:=0.0, Top:=0.0, Right:=8.0, Bottom:=0.0}
                                                Widget:=overlay:
                                                    Slots:=array:
                                                        #Weapon Icon - Pistols
                                                        overlay_slot:
                                                            VerticalAlignment:=vertical_alignment.Center
                                                            HorizontalAlignment:=horizontal_alignment.Center
                                                            Padding:=margin{Left:=0.0, Top:=0.0, Right:=0.0, Bottom:=0.0}
                                                            #Weapon icon BG
                                                            Widget:=texture_block{DefaultImage:=Textures.General.T_GenericSquare, DefaultDesiredSize:=vector2{X:=78.0, Y:=78.0}, DefaultTint:=MakeColorFromHex("#000c1d88")}

                                                        overlay_slot:
                                                            VerticalAlignment:=vertical_alignment.Center
                                                            HorizontalAlignment:=horizontal_alignment.Center
                                                            Padding:=margin{Left:=0.0, Top:=0.0, Right:=0.0, Bottom:=0.0}
                                                            #Weapon icon Border
                                                            Widget:=texture_block{DefaultImage:=Textures.General.T_GenericBorderSquare, DefaultDesiredSize:=vector2{X:=78.0, Y:=78.0}, DefaultTint:=MakeColorFromHex("#2e7be3FF")}

                                                        overlay_slot:
                                                            VerticalAlignment:=vertical_alignment.Top
                                                            HorizontalAlignment:=horizontal_alignment.Center
                                                            Padding:=margin{Left:=0.0, Top:=5.0, Right:=0.0, Bottom:=0.0}
                                                            #Weapon icon - pistol
                                                            Widget:=texture_block{DefaultImage:=Textures.Items.T_IconPistol, DefaultDesiredSize:=vector2{X:=50.0, Y:=50.0}, DefaultTint:=MakeColorFromHex("#FFFFFFFF")}

                                                        overlay_slot:
                                                            VerticalAlignment:=vertical_alignment.Bottom
                                                            HorizontalAlignment:=horizontal_alignment.Center
                                                            Padding:=margin{Left:=0.0, Top:=0.0, Right:=0.0, Bottom:=5.0}
                                                            #Weapon header - pistol
                                                            Widget:=texture_block{DefaultImage:=Textures.Items.T_HeaderPistol, DefaultDesiredSize:=vector2{X:=78.0, Y:=18.0}, DefaultTint:=MakeColorFromHex("#FFFFFFFF")}
                                                        
                                            stack_box_slot:
                                                VerticalAlignment:=vertical_alignment.Center
                                                HorizontalAlignment:=horizontal_alignment.Center
                                                Padding:=margin{Left:=0.0, Top:=0.0, Right:=8.0, Bottom:=0.0}
                                                Widget:=overlay:
                                                    Slots:=array:
                                                        #Weapon Icon - Shotgun
                                                        overlay_slot:
                                                            VerticalAlignment:=vertical_alignment.Center
                                                            HorizontalAlignment:=horizontal_alignment.Center
                                                            Padding:=margin{Left:=0.0, Top:=0.0, Right:=0.0, Bottom:=0.0}
                                                            #Weapon icon BG
                                                            Widget:=texture_block{DefaultImage:=Textures.General.T_GenericSquare, DefaultDesiredSize:=vector2{X:=78.0, Y:=78.0}, DefaultTint:=MakeColorFromHex("#000c1d88")}

                                                        overlay_slot:
                                                            VerticalAlignment:=vertical_alignment.Center
                                                            HorizontalAlignment:=horizontal_alignment.Center
                                                            Padding:=margin{Left:=0.0, Top:=0.0, Right:=0.0, Bottom:=0.0}
                                                            #Weapon icon Border
                                                            Widget:=texture_block{DefaultImage:=Textures.General.T_GenericBorderSquare, DefaultDesiredSize:=vector2{X:=78.0, Y:=78.0}, DefaultTint:=MakeColorFromHex("#2e7be3FF")}

                                                        overlay_slot:
                                                            VerticalAlignment:=vertical_alignment.Top
                                                            HorizontalAlignment:=horizontal_alignment.Center
                                                            Padding:=margin{Left:=0.0, Top:=5.0, Right:=0.0, Bottom:=0.0}
                                                            #Weapon icon - shotgun
                                                            Widget:=texture_block{DefaultImage:=Textures.Items.T_IconShotgun, DefaultDesiredSize:=vector2{X:=50.0, Y:=50.0}, DefaultTint:=MakeColorFromHex("#FFFFFFFF")}

                                                        overlay_slot:
                                                            VerticalAlignment:=vertical_alignment.Bottom
                                                            HorizontalAlignment:=horizontal_alignment.Center
                                                            Padding:=margin{Left:=0.0, Top:=0.0, Right:=0.0, Bottom:=5.0}
                                                            #Weapon header - shotgun
                                                            Widget:=texture_block{DefaultImage:=Textures.Items.T_HeaderShotgun, DefaultDesiredSize:=vector2{X:=78.0, Y:=18.0}, DefaultTint:=MakeColorFromHex("#FFFFFFFF")}

                                            stack_box_slot:
                                                VerticalAlignment:=vertical_alignment.Center
                                                HorizontalAlignment:=horizontal_alignment.Center
                                                Padding:=margin{Left:=0.0, Top:=0.0, Right:=8.0, Bottom:=0.0}
                                                Widget:=overlay:
                                                    Slots:=array:
                                                        #Weapon Icon - SMG
                                                        overlay_slot:
                                                            VerticalAlignment:=vertical_alignment.Top
                                                            HorizontalAlignment:=horizontal_alignment.Center
                                                            Padding:=margin{Left:=0.0, Top:=0.0, Right:=0.0, Bottom:=0.0}
                                                            #Weapon icon BG
                                                            Widget:=texture_block{DefaultImage:=Textures.General.T_GenericSquare, DefaultDesiredSize:=vector2{X:=78.0, Y:=78.0}, DefaultTint:=MakeColorFromHex("#000c1d88")}

                                                        overlay_slot:
                                                            VerticalAlignment:=vertical_alignment.Center
                                                            HorizontalAlignment:=horizontal_alignment.Center
                                                            Padding:=margin{Left:=0.0, Top:=0.0, Right:=0.0, Bottom:=0.0}
                                                            #Weapon icon Border
                                                            Widget:=texture_block{DefaultImage:=Textures.General.T_GenericBorderSquare, DefaultDesiredSize:=vector2{X:=78.0, Y:=78.0}, DefaultTint:=MakeColorFromHex("#2e7be3FF")}

                                                        overlay_slot:
                                                            VerticalAlignment:=vertical_alignment.Top
                                                            HorizontalAlignment:=horizontal_alignment.Center
                                                            Padding:=margin{Left:=0.0, Top:=5.0, Right:=0.0, Bottom:=0.0}
                                                            #Weapon icon - SMG
                                                            Widget:=texture_block{DefaultImage:=Textures.Items.T_IconSMG, DefaultDesiredSize:=vector2{X:=50.0, Y:=50.0}, DefaultTint:=MakeColorFromHex("#FFFFFFFF")}

                                                        overlay_slot:
                                                            VerticalAlignment:=vertical_alignment.Bottom
                                                            HorizontalAlignment:=horizontal_alignment.Center
                                                            Padding:=margin{Left:=0.0, Top:=0.0, Right:=0.0, Bottom:=5.0}
                                                            #Weapon header - SMG
                                                            Widget:=texture_block{DefaultImage:=Textures.Items.T_HeaderSMGs, DefaultDesiredSize:=vector2{X:=78.0, Y:=18.0}, DefaultTint:=MakeColorFromHex("#FFFFFFFF")}
                                                            

                                            stack_box_slot:
                                                VerticalAlignment:=vertical_alignment.Center
                                                HorizontalAlignment:=horizontal_alignment.Center
                                                Padding:=margin{Left:=0.0, Top:=0.0, Right:=8.0, Bottom:=0.0}
                                                Widget:=overlay:
                                                    Slots:=array:
                                                        #Weapon Icon - Rifle
                                                        overlay_slot:
                                                            VerticalAlignment:=vertical_alignment.Center
                                                            HorizontalAlignment:=horizontal_alignment.Center
                                                            Padding:=margin{Left:=0.0, Top:=0.0, Right:=0.0, Bottom:=0.0}
                                                            #Weapon icon BG
                                                            Widget:=texture_block{DefaultImage:=Textures.General.T_GenericSquare, DefaultDesiredSize:=vector2{X:=78.0, Y:=78.0}, DefaultTint:=MakeColorFromHex("#000c1d88")}

                                                        overlay_slot:
                                                            VerticalAlignment:=vertical_alignment.Top
                                                            HorizontalAlignment:=horizontal_alignment.Center
                                                            Padding:=margin{Left:=0.0, Top:=0.0, Right:=0.0, Bottom:=0.0}
                                                            #Weapon icon Border
                                                            Widget:=texture_block{DefaultImage:=Textures.General.T_GenericBorderSquare, DefaultDesiredSize:=vector2{X:=78.0, Y:=78.0}, DefaultTint:=MakeColorFromHex("#2e7be3FF")}

                                                        overlay_slot:
                                                            VerticalAlignment:=vertical_alignment.Top
                                                            HorizontalAlignment:=horizontal_alignment.Center
                                                            Padding:=margin{Left:=0.0, Top:=5.0, Right:=0.0, Bottom:=0.0}
                                                            #Weapon icon - Rifle
                                                            Widget:=texture_block{DefaultImage:=Textures.Items.T_IconRifle, DefaultDesiredSize:=vector2{X:=50.0, Y:=50.0}, DefaultTint:=MakeColorFromHex("#FFFFFFFF")}

                                                        overlay_slot:
                                                            VerticalAlignment:=vertical_alignment.Bottom
                                                            HorizontalAlignment:=horizontal_alignment.Center
                                                            Padding:=margin{Left:=0.0, Top:=0.0, Right:=0.0, Bottom:=5.0}
                                                            #Weapon header - Rifle
                                                            Widget:=texture_block{DefaultImage:=Textures.Items.T_HeaderRifle, DefaultDesiredSize:=vector2{X:=78.0, Y:=18.0}, DefaultTint:=MakeColorFromHex("#FFFFFFFF")}

                                            stack_box_slot:
                                                VerticalAlignment:=vertical_alignment.Center
                                                HorizontalAlignment:=horizontal_alignment.Center
                                                Padding:=margin{Left:=0.0, Top:=0.0, Right:=8.0, Bottom:=0.0}
                                                Widget:=overlay:
                                                    Slots:=array:
                                                        #Weapon Icon - Explosive
                                                        overlay_slot:
                                                            VerticalAlignment:=vertical_alignment.Center
                                                            HorizontalAlignment:=horizontal_alignment.Center
                                                            Padding:=margin{Left:=0.0, Top:=0.0, Right:=0.0, Bottom:=0.0}
                                                            #Weapon icon BG
                                                            Widget:=texture_block{DefaultImage:=Textures.General.T_GenericSquare, DefaultDesiredSize:=vector2{X:=78.0, Y:=78.0}, DefaultTint:=MakeColorFromHex("#000c1d88")}

                                                        overlay_slot:
                                                            VerticalAlignment:=vertical_alignment.Center
                                                            HorizontalAlignment:=horizontal_alignment.Center
                                                            Padding:=margin{Left:=0.0, Top:=0.0, Right:=0.0, Bottom:=0.0}
                                                            #Weapon icon Border
                                                            Widget:=texture_block{DefaultImage:=Textures.General.T_GenericBorderSquare, DefaultDesiredSize:=vector2{X:=78.0, Y:=78.0}, DefaultTint:=MakeColorFromHex("#2e7be3FF")}

                                                        overlay_slot:
                                                            VerticalAlignment:=vertical_alignment.Top
                                                            HorizontalAlignment:=horizontal_alignment.Center
                                                            Padding:=margin{Left:=0.0, Top:=5.0, Right:=0.0, Bottom:=0.0}
                                                            #Weapon icon - Explosive
                                                            Widget:=texture_block{DefaultImage:=Textures.Items.T_IconExplosive, DefaultDesiredSize:=vector2{X:=50.0, Y:=50.0}, DefaultTint:=MakeColorFromHex("#FFFFFFFF")}

                                                        overlay_slot:
                                                            VerticalAlignment:=vertical_alignment.Bottom
                                                            HorizontalAlignment:=horizontal_alignment.Center
                                                            Padding:=margin{Left:=0.0, Top:=0.0, Right:=0.0, Bottom:=5.0}
                                                            #Weapon header - Explosive
                                                            Widget:=texture_block{DefaultImage:=Textures.Items.T_HeaderExplosive, DefaultDesiredSize:=vector2{X:=78.0, Y:=18.0}, DefaultTint:=MakeColorFromHex("#FFFFFFFF")}

                                            stack_box_slot:
                                                VerticalAlignment:=vertical_alignment.Center
                                                HorizontalAlignment:=horizontal_alignment.Center
                                                Padding:=margin{Left:=0.0, Top:=0.0, Right:=8.0, Bottom:=0.0}
                                                Widget:=overlay:
                                                    Slots:=array:
                                                        #Weapon Icon - Special
                                                        overlay_slot:
                                                            VerticalAlignment:=vertical_alignment.Center
                                                            HorizontalAlignment:=horizontal_alignment.Center
                                                            Padding:=margin{Left:=0.0, Top:=0.0, Right:=0.0, Bottom:=0.0}
                                                            #Weapon icon BG
                                                            Widget:=texture_block{DefaultImage:=Textures.General.T_GenericSquare, DefaultDesiredSize:=vector2{X:=78.0, Y:=78.0}, DefaultTint:=MakeColorFromHex("#000c1d88")}

                                                        overlay_slot:
                                                            VerticalAlignment:=vertical_alignment.Center
                                                            HorizontalAlignment:=horizontal_alignment.Center
                                                            Padding:=margin{Left:=0.0, Top:=0.0, Right:=0.0, Bottom:=0.0}
                                                            #Weapon icon Border
                                                            Widget:=texture_block{DefaultImage:=Textures.General.T_GenericBorderSquare, DefaultDesiredSize:=vector2{X:=78.0, Y:=78.0}, DefaultTint:=MakeColorFromHex("#2e7be3FF")}

                                                        overlay_slot:
                                                            VerticalAlignment:=vertical_alignment.Top
                                                            HorizontalAlignment:=horizontal_alignment.Center
                                                            Padding:=margin{Left:=0.0, Top:=5.0, Right:=0.0, Bottom:=0.0}
                                                            #Weapon icon - special
                                                            Widget:=texture_block{DefaultImage:=Textures.Items.T_IconArtifact, DefaultDesiredSize:=vector2{X:=50.0, Y:=50.0}, DefaultTint:=MakeColorFromHex("#FFFFFFFF")}

                                                        overlay_slot:
                                                            VerticalAlignment:=vertical_alignment.Bottom
                                                            HorizontalAlignment:=horizontal_alignment.Center
                                                            Padding:=margin{Left:=0.0, Top:=0.0, Right:=0.0, Bottom:=5.0}
                                                            #Weapon header - Special
                                                            Widget:=texture_block{DefaultImage:=Textures.Items.T_HeaderSpecial, DefaultDesiredSize:=vector2{X:=78.0, Y:=18.0}, DefaultTint:=MakeColorFromHex("#FFFFFFFF")}

                                            stack_box_slot:
                                                VerticalAlignment:=vertical_alignment.Center
                                                HorizontalAlignment:=horizontal_alignment.Center
                                                Padding:=margin{Left:=0.0, Top:=0.0, Right:=8.0, Bottom:=0.0}
                                                Widget:=overlay:
                                                    Slots:=array:
                                                        #Weapon Icon - DMR Sniper
                                                        overlay_slot:
                                                            VerticalAlignment:=vertical_alignment.Center
                                                            HorizontalAlignment:=horizontal_alignment.Center
                                                            Padding:=margin{Left:=0.0, Top:=0.0, Right:=0.0, Bottom:=0.0}
                                                            #Weapon icon BG
                                                            Widget:=texture_block{DefaultImage:=Textures.General.T_GenericSquare, DefaultDesiredSize:=vector2{X:=78.0, Y:=78.0}, DefaultTint:=MakeColorFromHex("#000c1d88")}

                                                        overlay_slot:
                                                            VerticalAlignment:=vertical_alignment.Center
                                                            HorizontalAlignment:=horizontal_alignment.Center
                                                            Padding:=margin{Left:=0.0, Top:=0.0, Right:=0.0, Bottom:=0.0}
                                                            #Weapon icon Border
                                                            Widget:=texture_block{DefaultImage:=Textures.General.T_GenericBorderSquare, DefaultDesiredSize:=vector2{X:=78.0, Y:=78.0}, DefaultTint:=MakeColorFromHex("#2e7be3FF")}

                                                        overlay_slot:
                                                            VerticalAlignment:=vertical_alignment.Top
                                                            HorizontalAlignment:=horizontal_alignment.Center
                                                            Padding:=margin{Left:=0.0, Top:=5.0, Right:=0.0, Bottom:=0.0}
                                                            #Weapon icon - Sniper
                                                            Widget:=texture_block{DefaultImage:=Textures.Items.T_IconSniper, DefaultDesiredSize:=vector2{X:=50.0, Y:=50.0}, DefaultTint:=MakeColorFromHex("#FFFFFFFF")}

                                                        overlay_slot:
                                                            VerticalAlignment:=vertical_alignment.Bottom
                                                            HorizontalAlignment:=horizontal_alignment.Center
                                                            Padding:=margin{Left:=0.0, Top:=0.0, Right:=0.0, Bottom:=5.0}
                                                            #Weapon header - DMR Sniper
                                                            Widget:=texture_block{DefaultImage:=Textures.Items.T_HeaderDMRSniper, DefaultDesiredSize:=vector2{X:=78.0, Y:=18.0}, DefaultTint:=MakeColorFromHex("#FFFFFFFF")}

                                            stack_box_slot:
                                                VerticalAlignment:=vertical_alignment.Fill
                                                HorizontalAlignment:=horizontal_alignment.Center #Center
                                                Padding:=margin{Left:=0.0, Top:=0.0, Right:=0.0, Bottom:=0.0}
                                                Widget:=overlay:
                                                    Slots:=array:
                                                        overlay_slot:
                                                            #Makeshift size box
                                                            VerticalAlignment:=vertical_alignment.Center
                                                            HorizontalAlignment:=horizontal_alignment.Center
                                                            Padding:=margin{Left:=0.0, Top:=0.0, Right:=0.0, Bottom:=0.0}
                                                            Widget:=texture_block{DefaultImage:=Textures.General.T_GenSquTrans, DefaultDesiredSize:=vector2{X:=418.0, Y:=100.0}, DefaultTint:=MakeColorFromHex("#00FF0011")}

                                                        overlay_slot:
                                                            #Vertical stack box to add all the rarity information
                                                            VerticalAlignment:=vertical_alignment.Center
                                                            HorizontalAlignment:=horizontal_alignment.Center
                                                            Padding:=margin{Left:=0.0, Top:=0.0, Right:=0.0, Bottom:=0.0}
                                                            Widget:=stack_box:
                                                                Orientation:=orientation.Vertical
                                                                Slots:=array:
                                                                    stack_box_slot:
                                                                        #Horizontal Stack box for max rarity available
                                                                        VerticalAlignment:=vertical_alignment.Top
                                                                        HorizontalAlignment:=horizontal_alignment.Center
                                                                        Padding:=margin{Left:=0.0, Top:=0.0, Right:=0.0, Bottom:=0.0}
                                                                        Widget:=stack_box:
                                                                            Orientation:=orientation.Horizontal
                                                                            Slots:=array:
                                                                                stack_box_slot:
                                                                                    VerticalAlignment:=vertical_alignment.Center
                                                                                    HorizontalAlignment:=horizontal_alignment.Center
                                                                                    Padding:=margin{Left:=0.0, Top:=0.0, Right:=0.0, Bottom:=0.0}
                                                                                    Widget:=text_block{DefaultText := StringToMessage("Rarity chance up to "), DefaultTextColor:=MakeColorFromHex("#FFFFFFFF")}

                                                                                stack_box_slot:
                                                                                    VerticalAlignment:=vertical_alignment.Center
                                                                                    HorizontalAlignment:=horizontal_alignment.Center
                                                                                    Padding:=margin{Left:=0.0, Top:=0.0, Right:=0.0, Bottom:=0.0}
                                                                                    Widget:=texture_block{DefaultImage:=Textures.Rarity.T_FlagRare, DefaultDesiredSize:=vector2{X:=104.0, Y:=36.0}, DefaultTint:=MakeColorFromHex("#FFFFFFFF")}

                                                                    stack_box_slot:
                                                                        #Horizontal overlay for overall rarity scheme
                                                                        VerticalAlignment:=vertical_alignment.Bottom
                                                                        HorizontalAlignment:=horizontal_alignment.Center
                                                                        Padding:=margin{Left:=0.0, Top:=0.0, Right:=0.0, Bottom:=0.0}
                                                                        Widget:=overlay:
                                                                            Slots:=array:
                                                                                
                                                                                #Makeshift size box for the rarity scheme
                                                                                overlay_slot:
                                                                                    VerticalAlignment:=vertical_alignment.Center
                                                                                    HorizontalAlignment:=horizontal_alignment.Center
                                                                                    Padding:=margin{Left:=0.0, Top:=0.0, Right:=0.0, Bottom:=0.0}
                                                                                    Widget:=texture_block{DefaultImage:=Textures.General.T_GenSquTrans, DefaultDesiredSize:=vector2{X:=255.0, Y:=50.0}, DefaultTint:=MakeColorFromHex("#FF000022")}#Red frame

                                                                                #Visible border for rarity layout
                                                                                overlay_slot:
                                                                                    VerticalAlignment:=vertical_alignment.Center
                                                                                    HorizontalAlignment:=horizontal_alignment.Center
                                                                                    Padding:=margin{Left:=0.0, Top:=0.0, Right:=0.0, Bottom:=0.0}
                                                                                    Widget:=texture_block{DefaultImage:=Textures.General.T_GenericBorderSquare3, DefaultDesiredSize:=vector2{X:=255.0, Y:=40.0}, DefaultTint:=MakeColorFromHex("#2e7be3FF")}

                                                                                #Horizontal Stack box for rarity slots
                                                                                overlay_slot:
                                                                                    VerticalAlignment:=vertical_alignment.Fill
                                                                                    HorizontalAlignment:=horizontal_alignment.Fill
                                                                                    Padding:=margin{Left:=5.0, Top:=0.0, Right:=5.0, Bottom:=0.0}
                                                                                    Widget:=stack_box:
                                                                                        Orientation:=orientation.Horizontal
                                                                                        Slots:=array:
                                                                                            #Overlay for each rarity slots, including the pointing arrow
                                                                                            stack_box_slot:
                                                                                                VerticalAlignment:=vertical_alignment.Center
                                                                                                HorizontalAlignment:=horizontal_alignment.Center
                                                                                                Distribution:=option{1.0}
                                                                                                Padding:=margin{Left:=0.0, Top:=0.0, Right:=0.0, Bottom:=0.0}
                                                                                                Widget:=overlay:
                                                                                                    Slots:=array:
                                                                                                        #Makeshift size box
                                                                                                        overlay_slot:
                                                                                                            VerticalAlignment:=vertical_alignment.Center
                                                                                                            HorizontalAlignment:=horizontal_alignment.Center
                                                                                                            Padding:=margin{Left:=0.0, Top:=0.0, Right:=0.0, Bottom:=0.0}
                                                                                                            Widget:=texture_block{DefaultImage:=Textures.General.T_GenSquTrans, DefaultDesiredSize:=vector2{X:=30.0, Y:=40.0}, DefaultTint:=MakeColorFromHex("#0000FF08")}

                                                                                                        #Overlay for rarity border and texture
                                                                                                        overlay_slot:
                                                                                                            VerticalAlignment:=vertical_alignment.Center
                                                                                                            HorizontalAlignment:=horizontal_alignment.Center
                                                                                                            Padding:=margin{Left:=0.0, Top:=0.0, Right:=0.0, Bottom:=0.0}
                                                                                                            Widget:=overlay:
                                                                                                                Slots:=array:
                                                                                                                    
                                                                                                                    #Rarity texture border, when the rarity slot is empty
                                                                                                                    #overlay_slot:
                                                                                                                        #VerticalAlignment:=vertical_alignment.Center
                                                                                                                        #HorizontalAlignment:=horizontal_alignment.Center
                                                                                                                        #Padding:=margin{Left:=0.0, Top:=0.0, Right:=0.0, Bottom:=0.0}
                                                                                                                        #Widget:=texture_block{DefaultImage:=Textures.General.T_GenericBorderSquare2, DefaultDesiredSize:=vector2{X:=30.0, Y:=30.0}, DefaultTint:=MakeColorFromHex("#2e7be388")}

                                                                                                                    #Rarity texture
                                                                                                                    overlay_slot:
                                                                                                                        VerticalAlignment:=vertical_alignment.Center
                                                                                                                        HorizontalAlignment:=horizontal_alignment.Center
                                                                                                                        Padding:=margin{Left:=0.0, Top:=0.0, Right:=0.0, Bottom:=0.0}
                                                                                                                        Widget:=texture_block{DefaultImage:=Textures.General.T_GenericSquare, DefaultDesiredSize:=vector2{X:=20.0, Y:=20.0}, DefaultTint:=MakeColorFromHex("#979797FF")}
                                                                                                        
                                                                                                                    

                                                                                                        #Bracket
                                                                                                        overlay_slot:
                                                                                                            VerticalAlignment:=vertical_alignment.Fill
                                                                                                            HorizontalAlignment:=horizontal_alignment.Fill
                                                                                                            Padding:=margin{Left:=0.0, Top:=0.0, Right:=0.0, Bottom:=0.0}
                                                                                                            Widget:=stack_box:
                                                                                                                Orientation:=orientation.Horizontal
                                                                                                                #two instances of bracket texture to represent the range of rarities
                                                                                                                Slots:=array:

                                                                                                                    stack_box_slot:
                                                                                                                        VerticalAlignment:=vertical_alignment.Center
                                                                                                                        HorizontalAlignment:=horizontal_alignment.Left
                                                                                                                        Distribution:=option{1.0}
                                                                                                                        Padding:=margin{Left:=0.0, Top:=0.0, Right:=0.0, Bottom:=0.0}
                                                                                                                        Widget:=texture_block{DefaultImage:=Textures.General.T_RarityBracketLeft, DefaultDesiredSize:=vector2{X:=18.0, Y:=36.0}, DefaultTint:=MakeColorFromHex("#FFFFFF44")}

                                                                                                                    #stack_box_slot:
                                                                                                                        #VerticalAlignment:=vertical_alignment.Center
                                                                                                                        #HorizontalAlignment:=horizontal_alignment.Right
                                                                                                                        #Distribution:=option{1.0}
                                                                                                                        #Padding:=margin{Left:=0.0, Top:=0.0, Right:=0.0, Bottom:=0.0}
                                                                                                                        #Widget:=texture_block{DefaultImage:=Textures.General.T_RarityBracketRight, DefaultDesiredSize:=vector2{X:=18.0, Y:=36.0}, DefaultTint:=MakeColorFromHex("#FFFFFF44")}

                                                                                            stack_box_slot:
                                                                                                VerticalAlignment:=vertical_alignment.Center
                                                                                                HorizontalAlignment:=horizontal_alignment.Center
                                                                                                Distribution:=option{1.0}
                                                                                                Padding:=margin{Left:=0.0, Top:=0.0, Right:=0.0, Bottom:=0.0}
                                                                                                Widget:=overlay:
                                                                                                    Slots:=array:
                                                                                                        #Makeshift size box
                                                                                                        overlay_slot:
                                                                                                            VerticalAlignment:=vertical_alignment.Center
                                                                                                            HorizontalAlignment:=horizontal_alignment.Center
                                                                                                            Padding:=margin{Left:=0.0, Top:=0.0, Right:=0.0, Bottom:=0.0}
                                                                                                            Widget:=texture_block{DefaultImage:=Textures.General.T_GenSquTrans, DefaultDesiredSize:=vector2{X:=30.0, Y:=40.0}, DefaultTint:=MakeColorFromHex("#0000FF08")}

                                                                                                        #Overlay for rarity border and texture
                                                                                                        overlay_slot:
                                                                                                            VerticalAlignment:=vertical_alignment.Center
                                                                                                            HorizontalAlignment:=horizontal_alignment.Center
                                                                                                            Padding:=margin{Left:=0.0, Top:=0.0, Right:=0.0, Bottom:=0.0}
                                                                                                            Widget:=overlay:
                                                                                                                Slots:=array:
                                                                                                                    
                                                                                                                    #Rarity texture border, when the rarity slot is empty
                                                                                                                    #overlay_slot:
                                                                                                                        #VerticalAlignment:=vertical_alignment.Center
                                                                                                                        #HorizontalAlignment:=horizontal_alignment.Center
                                                                                                                        #Padding:=margin{Left:=0.0, Top:=0.0, Right:=0.0, Bottom:=0.0}
                                                                                                                        #Widget:=texture_block{DefaultImage:=Textures.General.T_GenericBorderSquare2, DefaultDesiredSize:=vector2{X:=30.0, Y:=30.0}, DefaultTint:=MakeColorFromHex("#2e7be388")}

                                                                                                                    #Rarity texture
                                                                                                                    overlay_slot:
                                                                                                                        VerticalAlignment:=vertical_alignment.Center
                                                                                                                        HorizontalAlignment:=horizontal_alignment.Center
                                                                                                                        Padding:=margin{Left:=0.0, Top:=0.0, Right:=0.0, Bottom:=0.0}
                                                                                                                        Widget:=texture_block{DefaultImage:=Textures.General.T_GenericSquare, DefaultDesiredSize:=vector2{X:=20.0, Y:=20.0}, DefaultTint:=MakeColorFromHex("#4ccc2aFF")}

                                                                                                        #Bracket
                                                                                                        overlay_slot:
                                                                                                            VerticalAlignment:=vertical_alignment.Fill
                                                                                                            HorizontalAlignment:=horizontal_alignment.Fill
                                                                                                            Padding:=margin{Left:=0.0, Top:=0.0, Right:=0.0, Bottom:=0.0}
                                                                                                            Widget:=stack_box:
                                                                                                                Orientation:=orientation.Horizontal
                                                                                                                #two instances of bracket texture to represent the range of rarities
                                                                                                                Slots:=array:

                                                                                                                    #stack_box_slot:
                                                                                                                        #VerticalAlignment:=vertical_alignment.Center
                                                                                                                        #HorizontalAlignment:=horizontal_alignment.Left
                                                                                                                        #Distribution:=option{1.0}
                                                                                                                        #Padding:=margin{Left:=0.0, Top:=0.0, Right:=0.0, Bottom:=0.0}
                                                                                                                        #Widget:=texture_block{DefaultImage:=Textures.General.T_RarityBracketLeft, DefaultDesiredSize:=vector2{X:=18.0, Y:=36.0}, DefaultTint:=MakeColorFromHex("#FFFFFF44")}

                                                                                                                    #stack_box_slot:
                                                                                                                        #VerticalAlignment:=vertical_alignment.Center
                                                                                                                        #HorizontalAlignment:=horizontal_alignment.Right
                                                                                                                        #Distribution:=option{1.0}
                                                                                                                        #Padding:=margin{Left:=0.0, Top:=0.0, Right:=0.0, Bottom:=0.0}
                                                                                                                        #Widget:=texture_block{DefaultImage:=Textures.General.T_RarityBracketRight, DefaultDesiredSize:=vector2{X:=18.0, Y:=36.0}, DefaultTint:=MakeColorFromHex("#FFFFFF44")}

                                                                                            stack_box_slot:
                                                                                                VerticalAlignment:=vertical_alignment.Center
                                                                                                HorizontalAlignment:=horizontal_alignment.Center
                                                                                                Distribution:=option{1.0}
                                                                                                Padding:=margin{Left:=0.0, Top:=0.0, Right:=0.0, Bottom:=0.0}
                                                                                                Widget:=overlay:
                                                                                                    Slots:=array:
                                                                                                        #Makeshift size box
                                                                                                        overlay_slot:
                                                                                                            VerticalAlignment:=vertical_alignment.Center
                                                                                                            HorizontalAlignment:=horizontal_alignment.Center
                                                                                                            Padding:=margin{Left:=0.0, Top:=0.0, Right:=0.0, Bottom:=0.0}
                                                                                                            Widget:=texture_block{DefaultImage:=Textures.General.T_GenSquTrans, DefaultDesiredSize:=vector2{X:=30.0, Y:=40.0}, DefaultTint:=MakeColorFromHex("#0000FF08")}

                                                                                                        #Overlay for rarity border and texture
                                                                                                        overlay_slot:
                                                                                                            VerticalAlignment:=vertical_alignment.Center
                                                                                                            HorizontalAlignment:=horizontal_alignment.Center
                                                                                                            Padding:=margin{Left:=0.0, Top:=0.0, Right:=0.0, Bottom:=0.0}
                                                                                                            Widget:=overlay:
                                                                                                                Slots:=array:
                                                                                                                                                                                                                
                                                                                                                    #Rarity texture border, when the rarity slot is empty
                                                                                                                    #overlay_slot:
                                                                                                                        #VerticalAlignment:=vertical_alignment.Center
                                                                                                                        #HorizontalAlignment:=horizontal_alignment.Center
                                                                                                                        #Padding:=margin{Left:=0.0, Top:=0.0, Right:=0.0, Bottom:=0.0}
                                                                                                                        #Widget:=texture_block{DefaultImage:=Textures.General.T_GenericBorderSquare2, DefaultDesiredSize:=vector2{X:=30.0, Y:=30.0}, DefaultTint:=MakeColorFromHex("#2e7be388")}
                                                                                                                                                                                                                            
                                                                                                                    #Rarity texture
                                                                                                                    overlay_slot:
                                                                                                                        VerticalAlignment:=vertical_alignment.Center
                                                                                                                        HorizontalAlignment:=horizontal_alignment.Center
                                                                                                                        Padding:=margin{Left:=0.0, Top:=0.0, Right:=0.0, Bottom:=0.0}
                                                                                                                        Widget:=texture_block{DefaultImage:=Textures.General.T_GenericSquare, DefaultDesiredSize:=vector2{X:=20.0, Y:=20.0}, DefaultTint:=MakeColorFromHex("#2a54ccFF")}
                                                                                                                    
                                                                                                        #Bracket
                                                                                                        overlay_slot:
                                                                                                            VerticalAlignment:=vertical_alignment.Fill
                                                                                                            HorizontalAlignment:=horizontal_alignment.Fill
                                                                                                            Padding:=margin{Left:=0.0, Top:=0.0, Right:=0.0, Bottom:=0.0}
                                                                                                            Widget:=stack_box:
                                                                                                                Orientation:=orientation.Horizontal
                                                                                                                #two instances of bracket texture to represent the range of rarities
                                                                                                                Slots:=array:

                                                                                                                    #stack_box_slot:
                                                                                                                        #VerticalAlignment:=vertical_alignment.Center
                                                                                                                        #HorizontalAlignment:=horizontal_alignment.Left
                                                                                                                        #Distribution:=option{1.0}
                                                                                                                        #Padding:=margin{Left:=0.0, Top:=0.0, Right:=0.0, Bottom:=0.0}
                                                                                                                        #Widget:=texture_block{DefaultImage:=Textures.General.T_RarityBracketLeft, DefaultDesiredSize:=vector2{X:=18.0, Y:=36.0}, DefaultTint:=MakeColorFromHex("#FFFFFF44")}

                                                                                                                    stack_box_slot:
                                                                                                                        VerticalAlignment:=vertical_alignment.Center
                                                                                                                        HorizontalAlignment:=horizontal_alignment.Right
                                                                                                                        Distribution:=option{1.0}
                                                                                                                        Padding:=margin{Left:=0.0, Top:=0.0, Right:=0.0, Bottom:=0.0}
                                                                                                                        Widget:=texture_block{DefaultImage:=Textures.General.T_RarityBracketRight, DefaultDesiredSize:=vector2{X:=18.0, Y:=36.0}, DefaultTint:=MakeColorFromHex("#FFFFFF44")}

                                                                                            stack_box_slot:
                                                                                                VerticalAlignment:=vertical_alignment.Center
                                                                                                HorizontalAlignment:=horizontal_alignment.Center
                                                                                                Distribution:=option{1.0}
                                                                                                Padding:=margin{Left:=0.0, Top:=0.0, Right:=0.0, Bottom:=0.0}
                                                                                                Widget:=overlay:
                                                                                                    Slots:=array:
                                                                                                        #Makeshift size box
                                                                                                        overlay_slot:
                                                                                                            VerticalAlignment:=vertical_alignment.Center
                                                                                                            HorizontalAlignment:=horizontal_alignment.Center
                                                                                                            Padding:=margin{Left:=0.0, Top:=0.0, Right:=0.0, Bottom:=0.0}
                                                                                                            Widget:=texture_block{DefaultImage:=Textures.General.T_GenSquTrans, DefaultDesiredSize:=vector2{X:=30.0, Y:=40.0}, DefaultTint:=MakeColorFromHex("#0000FF08")}

                                                                                                        #Overlay for rarity border and texture
                                                                                                        overlay_slot:
                                                                                                            VerticalAlignment:=vertical_alignment.Center
                                                                                                            HorizontalAlignment:=horizontal_alignment.Center
                                                                                                            Padding:=margin{Left:=0.0, Top:=0.0, Right:=0.0, Bottom:=0.0}
                                                                                                            Widget:=overlay:
                                                                                                                Slots:=array:

                                                                                                                    #Rarity texture border, when the rarity slot is empty
                                                                                                                    overlay_slot:
                                                                                                                        VerticalAlignment:=vertical_alignment.Center
                                                                                                                        HorizontalAlignment:=horizontal_alignment.Center
                                                                                                                        Padding:=margin{Left:=0.0, Top:=0.0, Right:=0.0, Bottom:=0.0}
                                                                                                                        Widget:=texture_block{DefaultImage:=Textures.General.T_GenericBorderSquare2, DefaultDesiredSize:=vector2{X:=30.0, Y:=30.0}, DefaultTint:=MakeColorFromHex("#2e7be388")}
                                                                                                                    
                                                                                                                    #Rarity texture
                                                                                                                    #overlay_slot:
                                                                                                                        #VerticalAlignment:=vertical_alignment.Center
                                                                                                                        #HorizontalAlignment:=horizontal_alignment.Center
                                                                                                                        #Padding:=margin{Left:=0.0, Top:=0.0, Right:=0.0, Bottom:=0.0}
                                                                                                                        #Widget:=texture_block{DefaultImage:=Textures.General.T_GenericSquare, DefaultDesiredSize:=vector2{X:=30.0, Y:=30.0}, DefaultTint:=MakeColorFromHex("#b32accFF")}
                                                                                                                    
                                                                                                        #Bracket
                                                                                                        overlay_slot:
                                                                                                            VerticalAlignment:=vertical_alignment.Fill
                                                                                                            HorizontalAlignment:=horizontal_alignment.Fill
                                                                                                            Padding:=margin{Left:=0.0, Top:=0.0, Right:=0.0, Bottom:=0.0}
                                                                                                            Widget:=stack_box:
                                                                                                                Orientation:=orientation.Horizontal
                                                                                                                #two instances of bracket texture to represent the range of rarities
                                                                                                                Slots:=array:

                                                                                                                    #stack_box_slot:
                                                                                                                        #VerticalAlignment:=vertical_alignment.Center
                                                                                                                        #HorizontalAlignment:=horizontal_alignment.Left
                                                                                                                        #Distribution:=option{1.0}
                                                                                                                        #Padding:=margin{Left:=0.0, Top:=0.0, Right:=0.0, Bottom:=0.0}
                                                                                                                        #Widget:=texture_block{DefaultImage:=Textures.General.T_RarityBracketLeft, DefaultDesiredSize:=vector2{X:=18.0, Y:=36.0}, DefaultTint:=MakeColorFromHex("#FFFFFF44")}

                                                                                                                    #stack_box_slot:
                                                                                                                        #VerticalAlignment:=vertical_alignment.Center
                                                                                                                        #HorizontalAlignment:=horizontal_alignment.Right
                                                                                                                        #Distribution:=option{1.0}
                                                                                                                        #Padding:=margin{Left:=0.0, Top:=0.0, Right:=0.0, Bottom:=0.0}
                                                                                                                        #Widget:=texture_block{DefaultImage:=Textures.General.T_RarityBracketRight, DefaultDesiredSize:=vector2{X:=18.0, Y:=36.0}, DefaultTint:=MakeColorFromHex("#FFFFFF44")}

                                                                                            stack_box_slot:
                                                                                                VerticalAlignment:=vertical_alignment.Center
                                                                                                HorizontalAlignment:=horizontal_alignment.Center
                                                                                                Distribution:=option{1.0}
                                                                                                Padding:=margin{Left:=0.0, Top:=0.0, Right:=0.0, Bottom:=0.0}
                                                                                                Widget:=overlay:
                                                                                                    Slots:=array:
                                                                                                        #Makeshift size box
                                                                                                        overlay_slot:
                                                                                                            VerticalAlignment:=vertical_alignment.Center
                                                                                                            HorizontalAlignment:=horizontal_alignment.Center
                                                                                                            Padding:=margin{Left:=0.0, Top:=0.0, Right:=0.0, Bottom:=0.0}
                                                                                                            Widget:=texture_block{DefaultImage:=Textures.General.T_GenSquTrans, DefaultDesiredSize:=vector2{X:=30.0, Y:=40.0}, DefaultTint:=MakeColorFromHex("#0000FF08")}

                                                                                                        #Overlay for rarity border and texture
                                                                                                        overlay_slot:
                                                                                                            VerticalAlignment:=vertical_alignment.Center
                                                                                                            HorizontalAlignment:=horizontal_alignment.Center
                                                                                                            Padding:=margin{Left:=0.0, Top:=0.0, Right:=0.0, Bottom:=0.0}
                                                                                                            Widget:=overlay:
                                                                                                                Slots:=array:

                                                                                                                    #Rarity texture border, when the rarity slot is empty
                                                                                                                    overlay_slot:
                                                                                                                        VerticalAlignment:=vertical_alignment.Center
                                                                                                                        HorizontalAlignment:=horizontal_alignment.Center
                                                                                                                        Padding:=margin{Left:=0.0, Top:=0.0, Right:=0.0, Bottom:=0.0}
                                                                                                                        Widget:=texture_block{DefaultImage:=Textures.General.T_GenericBorderSquare2, DefaultDesiredSize:=vector2{X:=30.0, Y:=30.0}, DefaultTint:=MakeColorFromHex("#2e7be388")}
                                                                                                                    
                                                                                                                    #Rarity texture
                                                                                                                    #overlay_slot:
                                                                                                                        #VerticalAlignment:=vertical_alignment.Center
                                                                                                                        #HorizontalAlignment:=horizontal_alignment.Center
                                                                                                                        #Padding:=margin{Left:=0.0, Top:=0.0, Right:=0.0, Bottom:=0.0}
                                                                                                                        #Widget:=texture_block{DefaultImage:=Textures.General.T_GenericSquare, DefaultDesiredSize:=vector2{X:=30.0, Y:=30.0}, DefaultTint:=MakeColorFromHex("#cc8d2aFF")}

                                                                                                        #Bracket
                                                                                                        overlay_slot:
                                                                                                            VerticalAlignment:=vertical_alignment.Fill
                                                                                                            HorizontalAlignment:=horizontal_alignment.Fill
                                                                                                            Padding:=margin{Left:=0.0, Top:=0.0, Right:=0.0, Bottom:=0.0}
                                                                                                            Widget:=stack_box:
                                                                                                                Orientation:=orientation.Horizontal
                                                                                                                #two instances of bracket texture to represent the range of rarities
                                                                                                                Slots:=array:

                                                                                                                    #stack_box_slot:
                                                                                                                        #VerticalAlignment:=vertical_alignment.Center
                                                                                                                        #HorizontalAlignment:=horizontal_alignment.Left
                                                                                                                        #Distribution:=option{1.0}
                                                                                                                        #Padding:=margin{Left:=0.0, Top:=0.0, Right:=0.0, Bottom:=0.0}
                                                                                                                        #Widget:=texture_block{DefaultImage:=Textures.General.T_RarityBracketLeft, DefaultDesiredSize:=vector2{X:=18.0, Y:=36.0}, DefaultTint:=MakeColorFromHex("#FFFFFF44")}

                                                                                                                    #stack_box_slot:
                                                                                                                        #VerticalAlignment:=vertical_alignment.Center
                                                                                                                        #HorizontalAlignment:=horizontal_alignment.Right
                                                                                                                        #Distribution:=option{1.0}
                                                                                                                        #Padding:=margin{Left:=0.0, Top:=0.0, Right:=0.0, Bottom:=0.0}
                                                                                                                        #Widget:=texture_block{DefaultImage:=Textures.General.T_RarityBracketRight, DefaultDesiredSize:=vector2{X:=18.0, Y:=36.0}, DefaultTint:=MakeColorFromHex("#FFFFFF44")}

                                                                                            stack_box_slot:
                                                                                                VerticalAlignment:=vertical_alignment.Center
                                                                                                HorizontalAlignment:=horizontal_alignment.Center
                                                                                                Distribution:=option{1.0}
                                                                                                Padding:=margin{Left:=0.0, Top:=0.0, Right:=0.0, Bottom:=0.0}
                                                                                                Widget:=overlay:
                                                                                                    Slots:=array:
                                                                                                        #Makeshift size box
                                                                                                        overlay_slot:
                                                                                                            VerticalAlignment:=vertical_alignment.Center
                                                                                                            HorizontalAlignment:=horizontal_alignment.Center
                                                                                                            Padding:=margin{Left:=0.0, Top:=0.0, Right:=0.0, Bottom:=0.0}
                                                                                                            Widget:=texture_block{DefaultImage:=Textures.General.T_GenSquTrans, DefaultDesiredSize:=vector2{X:=30.0, Y:=40.0}, DefaultTint:=MakeColorFromHex("#0000FF08")}

                                                                                                        #Overlay for rarity border and texture
                                                                                                        overlay_slot:
                                                                                                            VerticalAlignment:=vertical_alignment.Center
                                                                                                            HorizontalAlignment:=horizontal_alignment.Center
                                                                                                            Padding:=margin{Left:=0.0, Top:=0.0, Right:=0.0, Bottom:=0.0}
                                                                                                            Widget:=overlay:
                                                                                                                Slots:=array:
                                                                                                        
                                                                                                                    #Rarity texture border, when the rarity slot is empty
                                                                                                                    overlay_slot:
                                                                                                                        VerticalAlignment:=vertical_alignment.Center
                                                                                                                        HorizontalAlignment:=horizontal_alignment.Center
                                                                                                                        Padding:=margin{Left:=0.0, Top:=0.0, Right:=0.0, Bottom:=0.0}
                                                                                                                        Widget:=texture_block{DefaultImage:=Textures.General.T_GenericBorderSquare2, DefaultDesiredSize:=vector2{X:=30.0, Y:=30.0}, DefaultTint:=MakeColorFromHex("#2e7be388")}
                                                                                                                    
                                                                                                                    #Rarity texture
                                                                                                                    #overlay_slot:
                                                                                                                        #VerticalAlignment:=vertical_alignment.Center
                                                                                                                        #HorizontalAlignment:=horizontal_alignment.Center
                                                                                                                        #Padding:=margin{Left:=0.0, Top:=0.0, Right:=0.0, Bottom:=0.0}
                                                                                                                        #Widget:=texture_block{DefaultImage:=Textures.General.T_GenericSquare, DefaultDesiredSize:=vector2{X:=30.0, Y:=30.0}, DefaultTint:=MakeColorFromHex("#eacf19FF")}
                                                                                                                    
                                                                                                        #Bracket
                                                                                                        overlay_slot:
                                                                                                            VerticalAlignment:=vertical_alignment.Fill
                                                                                                            HorizontalAlignment:=horizontal_alignment.Fill
                                                                                                            Padding:=margin{Left:=0.0, Top:=0.0, Right:=0.0, Bottom:=0.0}
                                                                                                            Widget:=stack_box:
                                                                                                                Orientation:=orientation.Horizontal
                                                                                                                #two instances of bracket texture to represent the range of rarities
                                                                                                                Slots:=array:

                                                                                                                    #stack_box_slot:
                                                                                                                        #VerticalAlignment:=vertical_alignment.Center
                                                                                                                        #HorizontalAlignment:=horizontal_alignment.Left
                                                                                                                        #Distribution:=option{1.0}
                                                                                                                        #Padding:=margin{Left:=0.0, Top:=0.0, Right:=0.0, Bottom:=0.0}
                                                                                                                        #Widget:=texture_block{DefaultImage:=Textures.General.T_RarityBracketLeft, DefaultDesiredSize:=vector2{X:=18.0, Y:=36.0}, DefaultTint:=MakeColorFromHex("#FFFFFF44")}

                                                                                                                    #stack_box_slot:
                                                                                                                        #VerticalAlignment:=vertical_alignment.Center
                                                                                                                        #HorizontalAlignment:=horizontal_alignment.Right
                                                                                                                        #Distribution:=option{1.0}
                                                                                                                        #Padding:=margin{Left:=0.0, Top:=0.0, Right:=0.0, Bottom:=0.0}
                                                                                                                        #Widget:=texture_block{DefaultImage:=Textures.General.T_RarityBracketRight, DefaultDesiredSize:=vector2{X:=18.0, Y:=36.0}, DefaultTint:=MakeColorFromHex("#FFFFFF44")}

                    #overlay_slot:
                        #VerticalAlignment:=vertical_alignment.Fill
                        #HorizontalAlignment:=horizontal_alignment.Fill
                        #Padding:=margin{Left:=0.0, Top:=0.0, Right:=0.0, Bottom:=0.0}
                        #Widget:=texture_block{DefaultImage:=Textures.General.T_LevelPointer, DefaultDesiredSize:=vector2{X:=20.0, Y:=20.0}, DefaultTint:=MakeColorFromHex("#7de2FFFF")}
                                                                                
            #Definition for the Third level row
            LevelRow3:=overlay:
                Slots:=array:
                    overlay_slot:
                        #Makeshift size box
                        VerticalAlignment:=vertical_alignment.Center
                        HorizontalAlignment:=horizontal_alignment.Center
                        Padding:=margin{Left:=0.0, Top:=0.0, Right:=0.0, Bottom:=0.0}
                        Widget:=texture_block{DefaultImage:=Textures.General.T_GenSquTrans, DefaultDesiredSize:=vector2{X:=1250.0, Y:=114.0}, DefaultTint:=MakeColorFromHex("#0000FFFF")}

                    overlay_slot:
                        #Rarity level content
                        VerticalAlignment:=vertical_alignment.Center
                        HorizontalAlignment:=horizontal_alignment.Left
                        Padding:=margin{Left:=0.0, Top:=0.0, Right:=0.0, Bottom:=0.0}
                        Widget:=overlay:
                            Slots:=array:

                                overlay_slot:
                                    #Makeshift size box
                                    VerticalAlignment:=vertical_alignment.Center
                                    HorizontalAlignment:=horizontal_alignment.Center
                                    Padding:=margin{Left:=0.0, Top:=0.0, Right:=0.0, Bottom:=0.0}
                                    Widget:=texture_block{DefaultImage:=Textures.General.T_GenericSquare, DefaultDesiredSize:=vector2{X:=1190.0, Y:=114.0}, DefaultTint:=MakeColorFromHex("#010a1832")}

                                overlay_slot:
                                    #Highlight texture, over the content but under the rarity flag
                                    VerticalAlignment:=vertical_alignment.Center
                                    HorizontalAlignment:=horizontal_alignment.Center
                                    Padding:=margin{Left:=0.0, Top:=0.0, Right:=0.0, Bottom:=0.0}
                                    Widget:=texture_block{DefaultImage:=Textures.General.T_LevelBorder, DefaultDesiredSize:=vector2{X:=1190.0, Y:=114.0}, DefaultTint:=MakeColorFromHex("#174583FF")}

                                    #Horizontal stack box
                                overlay_slot:
                                    VerticalAlignment:=vertical_alignment.Fill
                                    HorizontalAlignment:=horizontal_alignment.Left
                                    Padding:=margin{Left:=0.0, Top:=0.0, Right:=0.0, Bottom:=0.0}
                                    Widget:=stack_box:
                                        Orientation:=orientation.Horizontal
                                        Slots:=array:
                                            #Level flag
                                            stack_box_slot:
                                                VerticalAlignment:=vertical_alignment.Top
                                                HorizontalAlignment:=horizontal_alignment.Left
                                                Padding:=margin{Left:=0.0, Top:=0.0, Right:=15.0, Bottom:=0.0}
                                                Widget:=overlay:

                                                    Slots:=array:
                                                        overlay_slot:
                                                            VerticalAlignment:=vertical_alignment.Center
                                                            HorizontalAlignment:=horizontal_alignment.Center
                                                            Padding:=margin{Left:=0.0, Top:=0.0, Right:=0.0, Bottom:=0.0}
                                                            #Tier level makeshift size box
                                                            Widget:=texture_block{DefaultImage:=Textures.General.T_GenSquTrans, DefaultDesiredSize:=vector2{X:=150.0, Y:=114.0}, DefaultTint:=MakeColorFromHex("#FF0000FF")}

                                                        overlay_slot:
                                                            VerticalAlignment:=vertical_alignment.Top
                                                            HorizontalAlignment:=horizontal_alignment.Left
                                                            Padding:=margin{Left:=0.0, Top:=0.0, Right:=0.0, Bottom:=0.0}
                                                            #Selected level yellow flag
                                                            Widget:=overlay:

                                                                #Slots:=array:
                                                                    #overlay_slot:
                                                                        #VerticalAlignment:=vertical_alignment.Center
                                                                        #HorizontalAlignment:=horizontal_alignment.Center
                                                                        #Padding:=margin{Left:=0.0, Top:=0.0, Right:=0.0, Bottom:=0.0}
                                                                        #level flag texture
                                                                        #Widget:=texture_block{DefaultImage:=Textures.General.T_GenLevelFlag, DefaultDesiredSize:=vector2{X:=140.0, Y:=45.0}, DefaultTint:=MakeColorFromHex("#f7ff1aFF")}

                                                                    #overlay_slot:
                                                                        #VerticalAlignment:=vertical_alignment.Center
                                                                        #HorizontalAlignment:=horizontal_alignment.Center
                                                                        #Padding:=margin{Left:=0.0, Top:=0.0, Right:=0.0, Bottom:=0.0}
                                                                        #level flag text
                                                                        #Widget:=text_block{DefaultText := StringToMessage("Level 0"), DefaultTextColor:=MakeColorFromHex("#000000FF")}

                                                        overlay_slot:
                                                            VerticalAlignment:=vertical_alignment.Fill
                                                            HorizontalAlignment:=horizontal_alignment.Center
                                                            Padding:=margin{Left:=0.0, Top:=5.0, Right:=0.0, Bottom:=5.0}
                                                            #Unselected level and lock
                                                            Widget:=overlay:
                                                                Slots:=array:
                                                                    
                                                                    overlay_slot:
                                                                        VerticalAlignment:=vertical_alignment.Top
                                                                        HorizontalAlignment:=horizontal_alignment.Center
                                                                        Padding:=margin{Left:=0.0, Top:=0.0, Right:=0.0, Bottom:=0.0}
                                                                        #level flag texture
                                                                        Widget:=text_block{DefaultText := StringToMessage("Level 3"), DefaultTextColor:=MakeColorFromHex("#a3a6aaFF")}

                                                                    overlay_slot:
                                                                        VerticalAlignment:=vertical_alignment.Bottom
                                                                        HorizontalAlignment:=horizontal_alignment.Center
                                                                        Padding:=margin{Left:=0.0, Top:=0.0, Right:=0.0, Bottom:=0.0}
                                                                        #level flag texture
                                                                        Widget:=texture_block{DefaultImage:=Textures.General.T_Lock, DefaultDesiredSize:=vector2{X:=52.0, Y:=58.0}, DefaultTint:=MakeColorFromHex("#a3a6aaff")}
                                            
                                            #Level weapon roster
                                            stack_box_slot:
                                                VerticalAlignment:=vertical_alignment.Center
                                                HorizontalAlignment:=horizontal_alignment.Center
                                                Padding:=margin{Left:=0.0, Top:=0.0, Right:=8.0, Bottom:=0.0}
                                                Widget:=overlay:
                                                    Slots:=array:
                                                        #Weapon Icon - Pistols
                                                        overlay_slot:
                                                            VerticalAlignment:=vertical_alignment.Center
                                                            HorizontalAlignment:=horizontal_alignment.Center
                                                            Padding:=margin{Left:=0.0, Top:=0.0, Right:=0.0, Bottom:=0.0}
                                                            #Weapon icon BG
                                                            Widget:=texture_block{DefaultImage:=Textures.General.T_GenericSquare, DefaultDesiredSize:=vector2{X:=78.0, Y:=78.0}, DefaultTint:=MakeColorFromHex("#000c1d88")}

                                                        overlay_slot:
                                                            VerticalAlignment:=vertical_alignment.Center
                                                            HorizontalAlignment:=horizontal_alignment.Center
                                                            Padding:=margin{Left:=0.0, Top:=0.0, Right:=0.0, Bottom:=0.0}
                                                            #Weapon icon Border
                                                            Widget:=texture_block{DefaultImage:=Textures.General.T_GenericBorderSquare, DefaultDesiredSize:=vector2{X:=78.0, Y:=78.0}, DefaultTint:=MakeColorFromHex("#174583FF")}

                                                        overlay_slot:
                                                            VerticalAlignment:=vertical_alignment.Top
                                                            HorizontalAlignment:=horizontal_alignment.Center
                                                            Padding:=margin{Left:=0.0, Top:=5.0, Right:=0.0, Bottom:=0.0}
                                                            #Weapon icon - pistol
                                                            Widget:=texture_block{DefaultImage:=Textures.Items.T_IconPistol, DefaultDesiredSize:=vector2{X:=50.0, Y:=50.0}, DefaultTint:=MakeColorFromHex("#a3a6aaff")}

                                                        overlay_slot:
                                                            VerticalAlignment:=vertical_alignment.Bottom
                                                            HorizontalAlignment:=horizontal_alignment.Center
                                                            Padding:=margin{Left:=0.0, Top:=0.0, Right:=0.0, Bottom:=5.0}
                                                            #Weapon header - pistol
                                                            Widget:=texture_block{DefaultImage:=Textures.Items.T_HeaderPistol, DefaultDesiredSize:=vector2{X:=78.0, Y:=18.0}, DefaultTint:=MakeColorFromHex("#a3a6aaff")}
                                                        
                                            stack_box_slot:
                                                VerticalAlignment:=vertical_alignment.Center
                                                HorizontalAlignment:=horizontal_alignment.Center
                                                Padding:=margin{Left:=0.0, Top:=0.0, Right:=8.0, Bottom:=0.0}
                                                Widget:=overlay:
                                                    Slots:=array:
                                                        #Weapon Icon - Shotgun
                                                        overlay_slot:
                                                            VerticalAlignment:=vertical_alignment.Center
                                                            HorizontalAlignment:=horizontal_alignment.Center
                                                            Padding:=margin{Left:=0.0, Top:=0.0, Right:=0.0, Bottom:=0.0}
                                                            #Weapon icon BG
                                                            Widget:=texture_block{DefaultImage:=Textures.General.T_GenericSquare, DefaultDesiredSize:=vector2{X:=78.0, Y:=78.0}, DefaultTint:=MakeColorFromHex("#000c1d88")}

                                                        overlay_slot:
                                                            VerticalAlignment:=vertical_alignment.Center
                                                            HorizontalAlignment:=horizontal_alignment.Center
                                                            Padding:=margin{Left:=0.0, Top:=0.0, Right:=0.0, Bottom:=0.0}
                                                            #Weapon icon Border
                                                            Widget:=texture_block{DefaultImage:=Textures.General.T_GenericBorderSquare, DefaultDesiredSize:=vector2{X:=78.0, Y:=78.0}, DefaultTint:=MakeColorFromHex("#174583FF")}

                                                        overlay_slot:
                                                            VerticalAlignment:=vertical_alignment.Top
                                                            HorizontalAlignment:=horizontal_alignment.Center
                                                            Padding:=margin{Left:=0.0, Top:=5.0, Right:=0.0, Bottom:=0.0}
                                                            #Weapon icon - shotgun
                                                            Widget:=texture_block{DefaultImage:=Textures.Items.T_IconShotgun, DefaultDesiredSize:=vector2{X:=50.0, Y:=50.0}, DefaultTint:=MakeColorFromHex("#a3a6aaff")}

                                                        overlay_slot:
                                                            VerticalAlignment:=vertical_alignment.Bottom
                                                            HorizontalAlignment:=horizontal_alignment.Center
                                                            Padding:=margin{Left:=0.0, Top:=0.0, Right:=0.0, Bottom:=5.0}
                                                            #Weapon header - shotgun
                                                            Widget:=texture_block{DefaultImage:=Textures.Items.T_HeaderShotgun, DefaultDesiredSize:=vector2{X:=78.0, Y:=18.0}, DefaultTint:=MakeColorFromHex("#a3a6aaff")}

                                            stack_box_slot:
                                                VerticalAlignment:=vertical_alignment.Center
                                                HorizontalAlignment:=horizontal_alignment.Center
                                                Padding:=margin{Left:=0.0, Top:=0.0, Right:=8.0, Bottom:=0.0}
                                                Widget:=overlay:
                                                    Slots:=array:
                                                        #Weapon Icon - SMG
                                                        overlay_slot:
                                                            VerticalAlignment:=vertical_alignment.Top
                                                            HorizontalAlignment:=horizontal_alignment.Center
                                                            Padding:=margin{Left:=0.0, Top:=0.0, Right:=0.0, Bottom:=0.0}
                                                            #Weapon icon BG
                                                            Widget:=texture_block{DefaultImage:=Textures.General.T_GenericSquare, DefaultDesiredSize:=vector2{X:=78.0, Y:=78.0}, DefaultTint:=MakeColorFromHex("#000c1d88")}

                                                        overlay_slot:
                                                            VerticalAlignment:=vertical_alignment.Center
                                                            HorizontalAlignment:=horizontal_alignment.Center
                                                            Padding:=margin{Left:=0.0, Top:=0.0, Right:=0.0, Bottom:=0.0}
                                                            #Weapon icon Border
                                                            Widget:=texture_block{DefaultImage:=Textures.General.T_GenericBorderSquare, DefaultDesiredSize:=vector2{X:=78.0, Y:=78.0}, DefaultTint:=MakeColorFromHex("#174583FF")}

                                                        overlay_slot:
                                                            VerticalAlignment:=vertical_alignment.Top
                                                            HorizontalAlignment:=horizontal_alignment.Center
                                                            Padding:=margin{Left:=0.0, Top:=5.0, Right:=0.0, Bottom:=0.0}
                                                            #Weapon icon - SMG
                                                            Widget:=texture_block{DefaultImage:=Textures.Items.T_IconSMG, DefaultDesiredSize:=vector2{X:=50.0, Y:=50.0}, DefaultTint:=MakeColorFromHex("#a3a6aaff")}

                                                        overlay_slot:
                                                            VerticalAlignment:=vertical_alignment.Bottom
                                                            HorizontalAlignment:=horizontal_alignment.Center
                                                            Padding:=margin{Left:=0.0, Top:=0.0, Right:=0.0, Bottom:=5.0}
                                                            #Weapon header - SMG
                                                            Widget:=texture_block{DefaultImage:=Textures.Items.T_HeaderSMGs, DefaultDesiredSize:=vector2{X:=78.0, Y:=18.0}, DefaultTint:=MakeColorFromHex("#a3a6aaff")}
                                                            

                                            stack_box_slot:
                                                VerticalAlignment:=vertical_alignment.Center
                                                HorizontalAlignment:=horizontal_alignment.Center
                                                Padding:=margin{Left:=0.0, Top:=0.0, Right:=8.0, Bottom:=0.0}
                                                Widget:=overlay:
                                                    Slots:=array:
                                                        #Weapon Icon - Rifle
                                                        overlay_slot:
                                                            VerticalAlignment:=vertical_alignment.Center
                                                            HorizontalAlignment:=horizontal_alignment.Center
                                                            Padding:=margin{Left:=0.0, Top:=0.0, Right:=0.0, Bottom:=0.0}
                                                            #Weapon icon BG
                                                            Widget:=texture_block{DefaultImage:=Textures.General.T_GenericSquare, DefaultDesiredSize:=vector2{X:=78.0, Y:=78.0}, DefaultTint:=MakeColorFromHex("#000c1d88")}

                                                        overlay_slot:
                                                            VerticalAlignment:=vertical_alignment.Top
                                                            HorizontalAlignment:=horizontal_alignment.Center
                                                            Padding:=margin{Left:=0.0, Top:=0.0, Right:=0.0, Bottom:=0.0}
                                                            #Weapon icon Border
                                                            Widget:=texture_block{DefaultImage:=Textures.General.T_GenericBorderSquare, DefaultDesiredSize:=vector2{X:=78.0, Y:=78.0}, DefaultTint:=MakeColorFromHex("#174583FF")}

                                                        overlay_slot:
                                                            VerticalAlignment:=vertical_alignment.Top
                                                            HorizontalAlignment:=horizontal_alignment.Center
                                                            Padding:=margin{Left:=0.0, Top:=5.0, Right:=0.0, Bottom:=0.0}
                                                            #Weapon icon - Rifle
                                                            Widget:=texture_block{DefaultImage:=Textures.Items.T_IconRifle, DefaultDesiredSize:=vector2{X:=50.0, Y:=50.0}, DefaultTint:=MakeColorFromHex("#a3a6aaff")}

                                                        overlay_slot:
                                                            VerticalAlignment:=vertical_alignment.Bottom
                                                            HorizontalAlignment:=horizontal_alignment.Center
                                                            Padding:=margin{Left:=0.0, Top:=0.0, Right:=0.0, Bottom:=5.0}
                                                            #Weapon header - Rifle
                                                            Widget:=texture_block{DefaultImage:=Textures.Items.T_HeaderRifle, DefaultDesiredSize:=vector2{X:=78.0, Y:=18.0}, DefaultTint:=MakeColorFromHex("#a3a6aaff")}

                                            stack_box_slot:
                                                VerticalAlignment:=vertical_alignment.Center
                                                HorizontalAlignment:=horizontal_alignment.Center
                                                Padding:=margin{Left:=0.0, Top:=0.0, Right:=8.0, Bottom:=0.0}
                                                Widget:=overlay:
                                                    Slots:=array:
                                                        #Weapon Icon - Explosive
                                                        overlay_slot:
                                                            VerticalAlignment:=vertical_alignment.Center
                                                            HorizontalAlignment:=horizontal_alignment.Center
                                                            Padding:=margin{Left:=0.0, Top:=0.0, Right:=0.0, Bottom:=0.0}
                                                            #Weapon icon BG
                                                            Widget:=texture_block{DefaultImage:=Textures.General.T_GenericSquare, DefaultDesiredSize:=vector2{X:=78.0, Y:=78.0}, DefaultTint:=MakeColorFromHex("#000c1d88")}

                                                        overlay_slot:
                                                            VerticalAlignment:=vertical_alignment.Center
                                                            HorizontalAlignment:=horizontal_alignment.Center
                                                            Padding:=margin{Left:=0.0, Top:=0.0, Right:=0.0, Bottom:=0.0}
                                                            #Weapon icon Border
                                                            Widget:=texture_block{DefaultImage:=Textures.General.T_GenericBorderSquare, DefaultDesiredSize:=vector2{X:=78.0, Y:=78.0}, DefaultTint:=MakeColorFromHex("#174583FF")}

                                                        overlay_slot:
                                                            VerticalAlignment:=vertical_alignment.Top
                                                            HorizontalAlignment:=horizontal_alignment.Center
                                                            Padding:=margin{Left:=0.0, Top:=5.0, Right:=0.0, Bottom:=0.0}
                                                            #Weapon icon - Explosive
                                                            Widget:=texture_block{DefaultImage:=Textures.Items.T_IconExplosive, DefaultDesiredSize:=vector2{X:=50.0, Y:=50.0}, DefaultTint:=MakeColorFromHex("#a3a6aaff")}

                                                        overlay_slot:
                                                            VerticalAlignment:=vertical_alignment.Bottom
                                                            HorizontalAlignment:=horizontal_alignment.Center
                                                            Padding:=margin{Left:=0.0, Top:=0.0, Right:=0.0, Bottom:=5.0}
                                                            #Weapon header - Explosive
                                                            Widget:=texture_block{DefaultImage:=Textures.Items.T_HeaderExplosive, DefaultDesiredSize:=vector2{X:=78.0, Y:=18.0}, DefaultTint:=MakeColorFromHex("#a3a6aaff")}

                                            stack_box_slot:
                                                VerticalAlignment:=vertical_alignment.Center
                                                HorizontalAlignment:=horizontal_alignment.Center
                                                Padding:=margin{Left:=0.0, Top:=0.0, Right:=8.0, Bottom:=0.0}
                                                Widget:=overlay:
                                                    Slots:=array:
                                                        #Weapon Icon - Special
                                                        overlay_slot:
                                                            VerticalAlignment:=vertical_alignment.Center
                                                            HorizontalAlignment:=horizontal_alignment.Center
                                                            Padding:=margin{Left:=0.0, Top:=0.0, Right:=0.0, Bottom:=0.0}
                                                            #Weapon icon BG
                                                            Widget:=texture_block{DefaultImage:=Textures.General.T_GenericSquare, DefaultDesiredSize:=vector2{X:=78.0, Y:=78.0}, DefaultTint:=MakeColorFromHex("#000c1d88")}

                                                        overlay_slot:
                                                            VerticalAlignment:=vertical_alignment.Center
                                                            HorizontalAlignment:=horizontal_alignment.Center
                                                            Padding:=margin{Left:=0.0, Top:=0.0, Right:=0.0, Bottom:=0.0}
                                                            #Weapon icon Border
                                                            Widget:=texture_block{DefaultImage:=Textures.General.T_GenericBorderSquare, DefaultDesiredSize:=vector2{X:=78.0, Y:=78.0}, DefaultTint:=MakeColorFromHex("#174583FF")}

                                                        overlay_slot:
                                                            VerticalAlignment:=vertical_alignment.Top
                                                            HorizontalAlignment:=horizontal_alignment.Center
                                                            Padding:=margin{Left:=0.0, Top:=5.0, Right:=0.0, Bottom:=0.0}
                                                            #Weapon icon - special
                                                            Widget:=texture_block{DefaultImage:=Textures.Items.T_IconArtifact, DefaultDesiredSize:=vector2{X:=50.0, Y:=50.0}, DefaultTint:=MakeColorFromHex("#a3a6aaff")}

                                                        overlay_slot:
                                                            VerticalAlignment:=vertical_alignment.Bottom
                                                            HorizontalAlignment:=horizontal_alignment.Center
                                                            Padding:=margin{Left:=0.0, Top:=0.0, Right:=0.0, Bottom:=5.0}
                                                            #Weapon header - Special
                                                            Widget:=texture_block{DefaultImage:=Textures.Items.T_HeaderSpecial, DefaultDesiredSize:=vector2{X:=78.0, Y:=18.0}, DefaultTint:=MakeColorFromHex("#a3a6aaff")}

                                            stack_box_slot:
                                                VerticalAlignment:=vertical_alignment.Center
                                                HorizontalAlignment:=horizontal_alignment.Center
                                                Padding:=margin{Left:=0.0, Top:=0.0, Right:=8.0, Bottom:=0.0}
                                                Widget:=overlay:
                                                    Slots:=array:
                                                        #Weapon Icon - DMR Sniper
                                                        overlay_slot:
                                                            VerticalAlignment:=vertical_alignment.Center
                                                            HorizontalAlignment:=horizontal_alignment.Center
                                                            Padding:=margin{Left:=0.0, Top:=0.0, Right:=0.0, Bottom:=0.0}
                                                            #Weapon icon BG
                                                            Widget:=texture_block{DefaultImage:=Textures.General.T_GenericSquare, DefaultDesiredSize:=vector2{X:=78.0, Y:=78.0}, DefaultTint:=MakeColorFromHex("#000c1d88")}

                                                        overlay_slot:
                                                            VerticalAlignment:=vertical_alignment.Center
                                                            HorizontalAlignment:=horizontal_alignment.Center
                                                            Padding:=margin{Left:=0.0, Top:=0.0, Right:=0.0, Bottom:=0.0}
                                                            #Weapon icon Border
                                                            Widget:=texture_block{DefaultImage:=Textures.General.T_GenericBorderSquare, DefaultDesiredSize:=vector2{X:=78.0, Y:=78.0}, DefaultTint:=MakeColorFromHex("#174583FF")}

                                                        overlay_slot:
                                                            VerticalAlignment:=vertical_alignment.Top
                                                            HorizontalAlignment:=horizontal_alignment.Center
                                                            Padding:=margin{Left:=0.0, Top:=5.0, Right:=0.0, Bottom:=0.0}
                                                            #Weapon icon - Sniper
                                                            Widget:=texture_block{DefaultImage:=Textures.Items.T_IconSniper, DefaultDesiredSize:=vector2{X:=50.0, Y:=50.0}, DefaultTint:=MakeColorFromHex("#a3a6aaff")}

                                                        overlay_slot:
                                                            VerticalAlignment:=vertical_alignment.Bottom
                                                            HorizontalAlignment:=horizontal_alignment.Center
                                                            Padding:=margin{Left:=0.0, Top:=0.0, Right:=0.0, Bottom:=5.0}
                                                            #Weapon header - DMR Sniper
                                                            Widget:=texture_block{DefaultImage:=Textures.Items.T_HeaderDMRSniper, DefaultDesiredSize:=vector2{X:=78.0, Y:=18.0}, DefaultTint:=MakeColorFromHex("#a3a6aaff")}

                                            stack_box_slot:
                                                VerticalAlignment:=vertical_alignment.Fill
                                                HorizontalAlignment:=horizontal_alignment.Center #Center
                                                Padding:=margin{Left:=0.0, Top:=0.0, Right:=0.0, Bottom:=0.0}
                                                Widget:=overlay:
                                                    Slots:=array:
                                                        overlay_slot:
                                                            #Makeshift size box
                                                            VerticalAlignment:=vertical_alignment.Center
                                                            HorizontalAlignment:=horizontal_alignment.Center
                                                            Padding:=margin{Left:=0.0, Top:=0.0, Right:=0.0, Bottom:=0.0}
                                                            Widget:=texture_block{DefaultImage:=Textures.General.T_GenSquTrans, DefaultDesiredSize:=vector2{X:=418.0, Y:=100.0}, DefaultTint:=MakeColorFromHex("#00FF0011")}

                                                        overlay_slot:
                                                            #Vertical stack box to add all the rarity information
                                                            VerticalAlignment:=vertical_alignment.Center
                                                            HorizontalAlignment:=horizontal_alignment.Center
                                                            Padding:=margin{Left:=0.0, Top:=0.0, Right:=0.0, Bottom:=0.0}
                                                            Widget:=stack_box:
                                                                Orientation:=orientation.Vertical
                                                                Slots:=array:
                                                                    stack_box_slot:
                                                                        #Horizontal Stack box for max rarity available
                                                                        VerticalAlignment:=vertical_alignment.Top
                                                                        HorizontalAlignment:=horizontal_alignment.Center
                                                                        Padding:=margin{Left:=0.0, Top:=0.0, Right:=0.0, Bottom:=0.0}
                                                                        Widget:=stack_box:
                                                                            Orientation:=orientation.Horizontal
                                                                            Slots:=array:
                                                                                stack_box_slot:
                                                                                    VerticalAlignment:=vertical_alignment.Center
                                                                                    HorizontalAlignment:=horizontal_alignment.Center
                                                                                    Padding:=margin{Left:=0.0, Top:=0.0, Right:=0.0, Bottom:=0.0}
                                                                                    Widget:=text_block{DefaultText := StringToMessage("Rarity chance up to "), DefaultTextColor:=MakeColorFromHex("#a3a6aaff")}

                                                                                stack_box_slot:
                                                                                    VerticalAlignment:=vertical_alignment.Center
                                                                                    HorizontalAlignment:=horizontal_alignment.Center
                                                                                    Padding:=margin{Left:=0.0, Top:=0.0, Right:=0.0, Bottom:=0.0}
                                                                                    Widget:=texture_block{DefaultImage:=Textures.Rarity.T_FlagEpic, DefaultDesiredSize:=vector2{X:=104.0, Y:=36.0}, DefaultTint:=MakeColorFromHex("#a3a6aaff")}

                                                                    stack_box_slot:
                                                                        #Horizontal overlay for overall rarity scheme
                                                                        VerticalAlignment:=vertical_alignment.Bottom
                                                                        HorizontalAlignment:=horizontal_alignment.Center
                                                                        Padding:=margin{Left:=0.0, Top:=0.0, Right:=0.0, Bottom:=0.0}
                                                                        Widget:=overlay:
                                                                            Slots:=array:
                                                                                
                                                                                #Makeshift size box for the rarity scheme
                                                                                overlay_slot:
                                                                                    VerticalAlignment:=vertical_alignment.Center
                                                                                    HorizontalAlignment:=horizontal_alignment.Center
                                                                                    Padding:=margin{Left:=0.0, Top:=0.0, Right:=0.0, Bottom:=0.0}
                                                                                    Widget:=texture_block{DefaultImage:=Textures.General.T_GenSquTrans, DefaultDesiredSize:=vector2{X:=255.0, Y:=50.0}, DefaultTint:=MakeColorFromHex("#FF000022")}#Red frame

                                                                                #Visible border for rarity layout
                                                                                overlay_slot:
                                                                                    VerticalAlignment:=vertical_alignment.Center
                                                                                    HorizontalAlignment:=horizontal_alignment.Center
                                                                                    Padding:=margin{Left:=0.0, Top:=0.0, Right:=0.0, Bottom:=0.0}
                                                                                    Widget:=texture_block{DefaultImage:=Textures.General.T_GenericBorderSquare3, DefaultDesiredSize:=vector2{X:=255.0, Y:=40.0}, DefaultTint:=MakeColorFromHex("#1d4b88FF")}

                                                                                #Horizontal Stack box for rarity slots
                                                                                overlay_slot:
                                                                                    VerticalAlignment:=vertical_alignment.Fill
                                                                                    HorizontalAlignment:=horizontal_alignment.Fill
                                                                                    Padding:=margin{Left:=5.0, Top:=0.0, Right:=5.0, Bottom:=0.0}
                                                                                    Widget:=stack_box:
                                                                                        Orientation:=orientation.Horizontal
                                                                                        Slots:=array:
                                                                                            #Overlay for each rarity slots, including the pointing arrow
                                                                                            stack_box_slot:
                                                                                                VerticalAlignment:=vertical_alignment.Center
                                                                                                HorizontalAlignment:=horizontal_alignment.Center
                                                                                                Distribution:=option{1.0}
                                                                                                Padding:=margin{Left:=0.0, Top:=0.0, Right:=0.0, Bottom:=0.0}
                                                                                                Widget:=overlay:
                                                                                                    Slots:=array:
                                                                                                        #Makeshift size box
                                                                                                        overlay_slot:
                                                                                                            VerticalAlignment:=vertical_alignment.Center
                                                                                                            HorizontalAlignment:=horizontal_alignment.Center
                                                                                                            Padding:=margin{Left:=0.0, Top:=0.0, Right:=0.0, Bottom:=0.0}
                                                                                                            Widget:=texture_block{DefaultImage:=Textures.General.T_GenSquTrans, DefaultDesiredSize:=vector2{X:=30.0, Y:=40.0}, DefaultTint:=MakeColorFromHex("#0000FF08")}

                                                                                                        #Overlay for rarity border and texture
                                                                                                        overlay_slot:
                                                                                                            VerticalAlignment:=vertical_alignment.Center
                                                                                                            HorizontalAlignment:=horizontal_alignment.Center
                                                                                                            Padding:=margin{Left:=0.0, Top:=0.0, Right:=0.0, Bottom:=0.0}
                                                                                                            Widget:=overlay:
                                                                                                                Slots:=array:
                                                                                                                    
                                                                                                                    #Rarity texture border, when the rarity slot is empty
                                                                                                                    overlay_slot:
                                                                                                                        VerticalAlignment:=vertical_alignment.Center
                                                                                                                        HorizontalAlignment:=horizontal_alignment.Center
                                                                                                                        Padding:=margin{Left:=0.0, Top:=0.0, Right:=0.0, Bottom:=0.0}
                                                                                                                        Widget:=texture_block{DefaultImage:=Textures.General.T_GenericBorderSquare2, DefaultDesiredSize:=vector2{X:=30.0, Y:=30.0}, DefaultTint:=MakeColorFromHex("#184584FF")}

                                                                                                                    #Rarity texture
                                                                                                                    #overlay_slot:
                                                                                                                        #VerticalAlignment:=vertical_alignment.Center
                                                                                                                        #HorizontalAlignment:=horizontal_alignment.Center
                                                                                                                        #Padding:=margin{Left:=0.0, Top:=0.0, Right:=0.0, Bottom:=0.0}
                                                                                                                        #Widget:=texture_block{DefaultImage:=Textures.General.T_GenericSquare, DefaultDesiredSize:=vector2{X:=30.0, Y:=30.0}, DefaultTint:=MakeColorFromHex("#979797FF")}
                                                                                                        
                                                                                                                    

                                                                                                        #Bracket
                                                                                                        overlay_slot:
                                                                                                            VerticalAlignment:=vertical_alignment.Fill
                                                                                                            HorizontalAlignment:=horizontal_alignment.Fill
                                                                                                            Padding:=margin{Left:=0.0, Top:=0.0, Right:=0.0, Bottom:=0.0}
                                                                                                            Widget:=stack_box:
                                                                                                                Orientation:=orientation.Horizontal
                                                                                                                #two instances of bracket texture to represent the range of rarities
                                                                                                                Slots:=array:

                                                                                                                    #stack_box_slot:
                                                                                                                        #VerticalAlignment:=vertical_alignment.Center
                                                                                                                        #HorizontalAlignment:=horizontal_alignment.Left
                                                                                                                        #Distribution:=option{1.0}
                                                                                                                        #Padding:=margin{Left:=0.0, Top:=0.0, Right:=0.0, Bottom:=0.0}
                                                                                                                        #Widget:=texture_block{DefaultImage:=Textures.General.T_RarityBracketLeft, DefaultDesiredSize:=vector2{X:=18.0, Y:=36.0}, DefaultTint:=MakeColorFromHex("#FFFFFF44")}

                                                                                                                    #stack_box_slot:
                                                                                                                        #VerticalAlignment:=vertical_alignment.Center
                                                                                                                        #HorizontalAlignment:=horizontal_alignment.Right
                                                                                                                        #Distribution:=option{1.0}
                                                                                                                        #Padding:=margin{Left:=0.0, Top:=0.0, Right:=0.0, Bottom:=0.0}
                                                                                                                        #Widget:=texture_block{DefaultImage:=Textures.General.T_RarityBracketRight, DefaultDesiredSize:=vector2{X:=18.0, Y:=36.0}, DefaultTint:=MakeColorFromHex("#FFFFFF44")}

                                                                                            stack_box_slot:
                                                                                                VerticalAlignment:=vertical_alignment.Center
                                                                                                HorizontalAlignment:=horizontal_alignment.Center
                                                                                                Distribution:=option{1.0}
                                                                                                Padding:=margin{Left:=0.0, Top:=0.0, Right:=0.0, Bottom:=0.0}
                                                                                                Widget:=overlay:
                                                                                                    Slots:=array:
                                                                                                        #Makeshift size box
                                                                                                        overlay_slot:
                                                                                                            VerticalAlignment:=vertical_alignment.Center
                                                                                                            HorizontalAlignment:=horizontal_alignment.Center
                                                                                                            Padding:=margin{Left:=0.0, Top:=0.0, Right:=0.0, Bottom:=0.0}
                                                                                                            Widget:=texture_block{DefaultImage:=Textures.General.T_GenSquTrans, DefaultDesiredSize:=vector2{X:=30.0, Y:=40.0}, DefaultTint:=MakeColorFromHex("#0000FF08")}

                                                                                                        #Overlay for rarity border and texture
                                                                                                        overlay_slot:
                                                                                                            VerticalAlignment:=vertical_alignment.Center
                                                                                                            HorizontalAlignment:=horizontal_alignment.Center
                                                                                                            Padding:=margin{Left:=0.0, Top:=0.0, Right:=0.0, Bottom:=0.0}
                                                                                                            Widget:=overlay:
                                                                                                                Slots:=array:
                                                                                                                    
                                                                                                                    #Rarity texture border, when the rarity slot is empty
                                                                                                                    #overlay_slot:
                                                                                                                        #VerticalAlignment:=vertical_alignment.Center
                                                                                                                        #HorizontalAlignment:=horizontal_alignment.Center
                                                                                                                        #Padding:=margin{Left:=0.0, Top:=0.0, Right:=0.0, Bottom:=0.0}
                                                                                                                        #Widget:=texture_block{DefaultImage:=Textures.General.T_GenericBorderSquare2, DefaultDesiredSize:=vector2{X:=30.0, Y:=30.0}, DefaultTint:=MakeColorFromHex("#2e7be388")}

                                                                                                                    #Rarity texture
                                                                                                                    overlay_slot:
                                                                                                                        VerticalAlignment:=vertical_alignment.Center
                                                                                                                        HorizontalAlignment:=horizontal_alignment.Center
                                                                                                                        Padding:=margin{Left:=0.0, Top:=0.0, Right:=0.0, Bottom:=0.0}
                                                                                                                        Widget:=texture_block{DefaultImage:=Textures.General.T_GenericSquare, DefaultDesiredSize:=vector2{X:=20.0, Y:=20.0}, DefaultTint:=MakeColorFromHex("#2a7028FF")}

                                                                                                        #Bracket
                                                                                                        overlay_slot:
                                                                                                            VerticalAlignment:=vertical_alignment.Fill
                                                                                                            HorizontalAlignment:=horizontal_alignment.Fill
                                                                                                            Padding:=margin{Left:=0.0, Top:=0.0, Right:=0.0, Bottom:=0.0}
                                                                                                            Widget:=stack_box:
                                                                                                                Orientation:=orientation.Horizontal
                                                                                                                #two instances of bracket texture to represent the range of rarities
                                                                                                                Slots:=array:

                                                                                                                    stack_box_slot:
                                                                                                                        VerticalAlignment:=vertical_alignment.Center
                                                                                                                        HorizontalAlignment:=horizontal_alignment.Left
                                                                                                                        Distribution:=option{1.0}
                                                                                                                        Padding:=margin{Left:=0.0, Top:=0.0, Right:=0.0, Bottom:=0.0}
                                                                                                                        Widget:=texture_block{DefaultImage:=Textures.General.T_RarityBracketLeft, DefaultDesiredSize:=vector2{X:=18.0, Y:=36.0}, DefaultTint:=MakeColorFromHex("#a3a6aaff")}

                                                                                                                    #stack_box_slot:
                                                                                                                        #VerticalAlignment:=vertical_alignment.Center
                                                                                                                        #HorizontalAlignment:=horizontal_alignment.Right
                                                                                                                        #Distribution:=option{1.0}
                                                                                                                        #Padding:=margin{Left:=0.0, Top:=0.0, Right:=0.0, Bottom:=0.0}
                                                                                                                        #Widget:=texture_block{DefaultImage:=Textures.General.T_RarityBracketRight, DefaultDesiredSize:=vector2{X:=18.0, Y:=36.0}, DefaultTint:=MakeColorFromHex("#FFFFFF44")}

                                                                                            stack_box_slot:
                                                                                                VerticalAlignment:=vertical_alignment.Center
                                                                                                HorizontalAlignment:=horizontal_alignment.Center
                                                                                                Distribution:=option{1.0}
                                                                                                Padding:=margin{Left:=0.0, Top:=0.0, Right:=0.0, Bottom:=0.0}
                                                                                                Widget:=overlay:
                                                                                                    Slots:=array:
                                                                                                        #Makeshift size box
                                                                                                        overlay_slot:
                                                                                                            VerticalAlignment:=vertical_alignment.Center
                                                                                                            HorizontalAlignment:=horizontal_alignment.Center
                                                                                                            Padding:=margin{Left:=0.0, Top:=0.0, Right:=0.0, Bottom:=0.0}
                                                                                                            Widget:=texture_block{DefaultImage:=Textures.General.T_GenSquTrans, DefaultDesiredSize:=vector2{X:=30.0, Y:=40.0}, DefaultTint:=MakeColorFromHex("#0000FF08")}

                                                                                                        #Overlay for rarity border and texture
                                                                                                        overlay_slot:
                                                                                                            VerticalAlignment:=vertical_alignment.Center
                                                                                                            HorizontalAlignment:=horizontal_alignment.Center
                                                                                                            Padding:=margin{Left:=0.0, Top:=0.0, Right:=0.0, Bottom:=0.0}
                                                                                                            Widget:=overlay:
                                                                                                                Slots:=array:
                                                                                                                                                                                                                
                                                                                                                    #Rarity texture border, when the rarity slot is empty
                                                                                                                    #overlay_slot:
                                                                                                                        #VerticalAlignment:=vertical_alignment.Center
                                                                                                                        #HorizontalAlignment:=horizontal_alignment.Center
                                                                                                                        #Padding:=margin{Left:=0.0, Top:=0.0, Right:=0.0, Bottom:=0.0}
                                                                                                                        #Widget:=texture_block{DefaultImage:=Textures.General.T_GenericBorderSquare2, DefaultDesiredSize:=vector2{X:=30.0, Y:=30.0}, DefaultTint:=MakeColorFromHex("#2e7be388")}
                                                                                                                                                                                                                            
                                                                                                                    #Rarity texture
                                                                                                                    overlay_slot:
                                                                                                                        VerticalAlignment:=vertical_alignment.Center
                                                                                                                        HorizontalAlignment:=horizontal_alignment.Center
                                                                                                                        Padding:=margin{Left:=0.0, Top:=0.0, Right:=0.0, Bottom:=0.0}
                                                                                                                        Widget:=texture_block{DefaultImage:=Textures.General.T_GenericSquare, DefaultDesiredSize:=vector2{X:=20.0, Y:=20.0}, DefaultTint:=MakeColorFromHex("#183379FF")}
                                                                                                                    
                                                                                                        #Bracket
                                                                                                        overlay_slot:
                                                                                                            VerticalAlignment:=vertical_alignment.Fill
                                                                                                            HorizontalAlignment:=horizontal_alignment.Fill
                                                                                                            Padding:=margin{Left:=0.0, Top:=0.0, Right:=0.0, Bottom:=0.0}
                                                                                                            Widget:=stack_box:
                                                                                                                Orientation:=orientation.Horizontal
                                                                                                                #two instances of bracket texture to represent the range of rarities
                                                                                                                Slots:=array:

                                                                                                                    #stack_box_slot:
                                                                                                                        #VerticalAlignment:=vertical_alignment.Center
                                                                                                                        #HorizontalAlignment:=horizontal_alignment.Left
                                                                                                                        #Distribution:=option{1.0}
                                                                                                                        #Padding:=margin{Left:=0.0, Top:=0.0, Right:=0.0, Bottom:=0.0}
                                                                                                                        #Widget:=texture_block{DefaultImage:=Textures.General.T_RarityBracketLeft, DefaultDesiredSize:=vector2{X:=18.0, Y:=36.0}, DefaultTint:=MakeColorFromHex("#FFFFFF44")}

                                                                                                                    #stack_box_slot:
                                                                                                                        #VerticalAlignment:=vertical_alignment.Center
                                                                                                                        #HorizontalAlignment:=horizontal_alignment.Right
                                                                                                                        #Distribution:=option{1.0}
                                                                                                                        #Padding:=margin{Left:=0.0, Top:=0.0, Right:=0.0, Bottom:=0.0}
                                                                                                                        #Widget:=texture_block{DefaultImage:=Textures.General.T_RarityBracketRight, DefaultDesiredSize:=vector2{X:=18.0, Y:=36.0}, DefaultTint:=MakeColorFromHex("#FFFFFF44")}

                                                                                            stack_box_slot:
                                                                                                VerticalAlignment:=vertical_alignment.Center
                                                                                                HorizontalAlignment:=horizontal_alignment.Center
                                                                                                Distribution:=option{1.0}
                                                                                                Padding:=margin{Left:=0.0, Top:=0.0, Right:=0.0, Bottom:=0.0}
                                                                                                Widget:=overlay:
                                                                                                    Slots:=array:
                                                                                                        #Makeshift size box
                                                                                                        overlay_slot:
                                                                                                            VerticalAlignment:=vertical_alignment.Center
                                                                                                            HorizontalAlignment:=horizontal_alignment.Center
                                                                                                            Padding:=margin{Left:=0.0, Top:=0.0, Right:=0.0, Bottom:=0.0}
                                                                                                            Widget:=texture_block{DefaultImage:=Textures.General.T_GenSquTrans, DefaultDesiredSize:=vector2{X:=30.0, Y:=40.0}, DefaultTint:=MakeColorFromHex("#0000FF08")}

                                                                                                        #Overlay for rarity border and texture
                                                                                                        overlay_slot:
                                                                                                            VerticalAlignment:=vertical_alignment.Center
                                                                                                            HorizontalAlignment:=horizontal_alignment.Center
                                                                                                            Padding:=margin{Left:=0.0, Top:=0.0, Right:=0.0, Bottom:=0.0}
                                                                                                            Widget:=overlay:
                                                                                                                Slots:=array:

                                                                                                                    #Rarity texture border, when the rarity slot is empty
                                                                                                                    #overlay_slot:
                                                                                                                        #VerticalAlignment:=vertical_alignment.Center
                                                                                                                        #HorizontalAlignment:=horizontal_alignment.Center
                                                                                                                        #Padding:=margin{Left:=0.0, Top:=0.0, Right:=0.0, Bottom:=0.0}
                                                                                                                        #Widget:=texture_block{DefaultImage:=Textures.General.T_GenericBorderSquare2, DefaultDesiredSize:=vector2{X:=30.0, Y:=30.0}, DefaultTint:=MakeColorFromHex("#2e7be388")}
                                                                                                                    
                                                                                                                    #Rarity texture
                                                                                                                    overlay_slot:
                                                                                                                        VerticalAlignment:=vertical_alignment.Center
                                                                                                                        HorizontalAlignment:=horizontal_alignment.Center
                                                                                                                        Padding:=margin{Left:=0.0, Top:=0.0, Right:=0.0, Bottom:=0.0}
                                                                                                                        Widget:=texture_block{DefaultImage:=Textures.General.T_GenericSquare, DefaultDesiredSize:=vector2{X:=20.0, Y:=20.0}, DefaultTint:=MakeColorFromHex("#5b1e79FF")}
                                                                                                                    
                                                                                                        #Bracket
                                                                                                        overlay_slot:
                                                                                                            VerticalAlignment:=vertical_alignment.Fill
                                                                                                            HorizontalAlignment:=horizontal_alignment.Fill
                                                                                                            Padding:=margin{Left:=0.0, Top:=0.0, Right:=0.0, Bottom:=0.0}
                                                                                                            Widget:=stack_box:
                                                                                                                Orientation:=orientation.Horizontal
                                                                                                                #two instances of bracket texture to represent the range of rarities
                                                                                                                Slots:=array:

                                                                                                                    #stack_box_slot:
                                                                                                                        #VerticalAlignment:=vertical_alignment.Center
                                                                                                                        #HorizontalAlignment:=horizontal_alignment.Left
                                                                                                                        #Distribution:=option{1.0}
                                                                                                                        #Padding:=margin{Left:=0.0, Top:=0.0, Right:=0.0, Bottom:=0.0}
                                                                                                                        #Widget:=texture_block{DefaultImage:=Textures.General.T_RarityBracketLeft, DefaultDesiredSize:=vector2{X:=18.0, Y:=36.0}, DefaultTint:=MakeColorFromHex("#FFFFFF44")}

                                                                                                                    stack_box_slot:
                                                                                                                        VerticalAlignment:=vertical_alignment.Center
                                                                                                                        HorizontalAlignment:=horizontal_alignment.Right
                                                                                                                        Distribution:=option{1.0}
                                                                                                                        Padding:=margin{Left:=0.0, Top:=0.0, Right:=0.0, Bottom:=0.0}
                                                                                                                        Widget:=texture_block{DefaultImage:=Textures.General.T_RarityBracketRight, DefaultDesiredSize:=vector2{X:=18.0, Y:=36.0}, DefaultTint:=MakeColorFromHex("#a3a6aaff")}

                                                                                            stack_box_slot:
                                                                                                VerticalAlignment:=vertical_alignment.Center
                                                                                                HorizontalAlignment:=horizontal_alignment.Center
                                                                                                Distribution:=option{1.0}
                                                                                                Padding:=margin{Left:=0.0, Top:=0.0, Right:=0.0, Bottom:=0.0}
                                                                                                Widget:=overlay:
                                                                                                    Slots:=array:
                                                                                                        #Makeshift size box
                                                                                                        overlay_slot:
                                                                                                            VerticalAlignment:=vertical_alignment.Center
                                                                                                            HorizontalAlignment:=horizontal_alignment.Center
                                                                                                            Padding:=margin{Left:=0.0, Top:=0.0, Right:=0.0, Bottom:=0.0}
                                                                                                            Widget:=texture_block{DefaultImage:=Textures.General.T_GenSquTrans, DefaultDesiredSize:=vector2{X:=30.0, Y:=40.0}, DefaultTint:=MakeColorFromHex("#0000FF08")}

                                                                                                        #Overlay for rarity border and texture
                                                                                                        overlay_slot:
                                                                                                            VerticalAlignment:=vertical_alignment.Center
                                                                                                            HorizontalAlignment:=horizontal_alignment.Center
                                                                                                            Padding:=margin{Left:=0.0, Top:=0.0, Right:=0.0, Bottom:=0.0}
                                                                                                            Widget:=overlay:
                                                                                                                Slots:=array:

                                                                                                                    #Rarity texture border, when the rarity slot is empty
                                                                                                                    overlay_slot:
                                                                                                                        VerticalAlignment:=vertical_alignment.Center
                                                                                                                        HorizontalAlignment:=horizontal_alignment.Center
                                                                                                                        Padding:=margin{Left:=0.0, Top:=0.0, Right:=0.0, Bottom:=0.0}
                                                                                                                        Widget:=texture_block{DefaultImage:=Textures.General.T_GenericBorderSquare2, DefaultDesiredSize:=vector2{X:=30.0, Y:=30.0}, DefaultTint:=MakeColorFromHex("#174583FF")}
                                                                                                                    
                                                                                                                    #Rarity texture
                                                                                                                    #overlay_slot:
                                                                                                                        #VerticalAlignment:=vertical_alignment.Center
                                                                                                                        #HorizontalAlignment:=horizontal_alignment.Center
                                                                                                                        #Padding:=margin{Left:=0.0, Top:=0.0, Right:=0.0, Bottom:=0.0}
                                                                                                                        #Widget:=texture_block{DefaultImage:=Textures.General.T_GenericSquare, DefaultDesiredSize:=vector2{X:=30.0, Y:=30.0}, DefaultTint:=MakeColorFromHex("#cc8d2aFF")}

                                                                                                        #Bracket
                                                                                                        overlay_slot:
                                                                                                            VerticalAlignment:=vertical_alignment.Fill
                                                                                                            HorizontalAlignment:=horizontal_alignment.Fill
                                                                                                            Padding:=margin{Left:=0.0, Top:=0.0, Right:=0.0, Bottom:=0.0}
                                                                                                            Widget:=stack_box:
                                                                                                                Orientation:=orientation.Horizontal
                                                                                                                #two instances of bracket texture to represent the range of rarities
                                                                                                                Slots:=array:

                                                                                                                    #stack_box_slot:
                                                                                                                        #VerticalAlignment:=vertical_alignment.Center
                                                                                                                        #HorizontalAlignment:=horizontal_alignment.Left
                                                                                                                        #Distribution:=option{1.0}
                                                                                                                        #Padding:=margin{Left:=0.0, Top:=0.0, Right:=0.0, Bottom:=0.0}
                                                                                                                        #Widget:=texture_block{DefaultImage:=Textures.General.T_RarityBracketLeft, DefaultDesiredSize:=vector2{X:=18.0, Y:=36.0}, DefaultTint:=MakeColorFromHex("#FFFFFF44")}

                                                                                                                    #stack_box_slot:
                                                                                                                        #VerticalAlignment:=vertical_alignment.Center
                                                                                                                        #HorizontalAlignment:=horizontal_alignment.Right
                                                                                                                        #Distribution:=option{1.0}
                                                                                                                        #Padding:=margin{Left:=0.0, Top:=0.0, Right:=0.0, Bottom:=0.0}
                                                                                                                        #Widget:=texture_block{DefaultImage:=Textures.General.T_RarityBracketRight, DefaultDesiredSize:=vector2{X:=18.0, Y:=36.0}, DefaultTint:=MakeColorFromHex("#FFFFFF44")}

                                                                                            stack_box_slot:
                                                                                                VerticalAlignment:=vertical_alignment.Center
                                                                                                HorizontalAlignment:=horizontal_alignment.Center
                                                                                                Distribution:=option{1.0}
                                                                                                Padding:=margin{Left:=0.0, Top:=0.0, Right:=0.0, Bottom:=0.0}
                                                                                                Widget:=overlay:
                                                                                                    Slots:=array:
                                                                                                        #Makeshift size box
                                                                                                        overlay_slot:
                                                                                                            VerticalAlignment:=vertical_alignment.Center
                                                                                                            HorizontalAlignment:=horizontal_alignment.Center
                                                                                                            Padding:=margin{Left:=0.0, Top:=0.0, Right:=0.0, Bottom:=0.0}
                                                                                                            Widget:=texture_block{DefaultImage:=Textures.General.T_GenSquTrans, DefaultDesiredSize:=vector2{X:=30.0, Y:=40.0}, DefaultTint:=MakeColorFromHex("#0000FF08")}

                                                                                                        #Overlay for rarity border and texture
                                                                                                        overlay_slot:
                                                                                                            VerticalAlignment:=vertical_alignment.Center
                                                                                                            HorizontalAlignment:=horizontal_alignment.Center
                                                                                                            Padding:=margin{Left:=0.0, Top:=0.0, Right:=0.0, Bottom:=0.0}
                                                                                                            Widget:=overlay:
                                                                                                                Slots:=array:
                                                                                                        
                                                                                                                    #Rarity texture border, when the rarity slot is empty
                                                                                                                    overlay_slot:
                                                                                                                        VerticalAlignment:=vertical_alignment.Center
                                                                                                                        HorizontalAlignment:=horizontal_alignment.Center
                                                                                                                        Padding:=margin{Left:=0.0, Top:=0.0, Right:=0.0, Bottom:=0.0}
                                                                                                                        Widget:=texture_block{DefaultImage:=Textures.General.T_GenericBorderSquare2, DefaultDesiredSize:=vector2{X:=30.0, Y:=30.0}, DefaultTint:=MakeColorFromHex("#174583FF")}
                                                                                                                    
                                                                                                                    #Rarity texture
                                                                                                                    #overlay_slot:
                                                                                                                        #VerticalAlignment:=vertical_alignment.Center
                                                                                                                        #HorizontalAlignment:=horizontal_alignment.Center
                                                                                                                        #Padding:=margin{Left:=0.0, Top:=0.0, Right:=0.0, Bottom:=0.0}
                                                                                                                        #Widget:=texture_block{DefaultImage:=Textures.General.T_GenericSquare, DefaultDesiredSize:=vector2{X:=30.0, Y:=30.0}, DefaultTint:=MakeColorFromHex("#eacf19FF")}
                                                                                                                    
                                                                                                        #Bracket
                                                                                                        overlay_slot:
                                                                                                            VerticalAlignment:=vertical_alignment.Fill
                                                                                                            HorizontalAlignment:=horizontal_alignment.Fill
                                                                                                            Padding:=margin{Left:=0.0, Top:=0.0, Right:=0.0, Bottom:=0.0}
                                                                                                            Widget:=stack_box:
                                                                                                                Orientation:=orientation.Horizontal
                                                                                                                #two instances of bracket texture to represent the range of rarities
                                                                                                                Slots:=array:

                                                                                                                    #stack_box_slot:
                                                                                                                        #VerticalAlignment:=vertical_alignment.Center
                                                                                                                        #HorizontalAlignment:=horizontal_alignment.Left
                                                                                                                        #Distribution:=option{1.0}
                                                                                                                        #Padding:=margin{Left:=0.0, Top:=0.0, Right:=0.0, Bottom:=0.0}
                                                                                                                        #Widget:=texture_block{DefaultImage:=Textures.General.T_RarityBracketLeft, DefaultDesiredSize:=vector2{X:=18.0, Y:=36.0}, DefaultTint:=MakeColorFromHex("#FFFFFF44")}

                                                                                                                    #stack_box_slot:
                                                                                                                        #VerticalAlignment:=vertical_alignment.Center
                                                                                                                        #HorizontalAlignment:=horizontal_alignment.Right
                                                                                                                        #Distribution:=option{1.0}
                                                                                                                        #Padding:=margin{Left:=0.0, Top:=0.0, Right:=0.0, Bottom:=0.0}
                                                                                                                        #Widget:=texture_block{DefaultImage:=Textures.General.T_RarityBracketRight, DefaultDesiredSize:=vector2{X:=18.0, Y:=36.0}, DefaultTint:=MakeColorFromHex("#FFFFFF44")}

                    #overlay_slot:
                        #VerticalAlignment:=vertical_alignment.Fill
                        #HorizontalAlignment:=horizontal_alignment.Fill
                        #Padding:=margin{Left:=0.0, Top:=0.0, Right:=0.0, Bottom:=0.0}
                        #Widget:=texture_block{DefaultImage:=Textures.General.T_LevelPointer, DefaultDesiredSize:=vector2{X:=20.0, Y:=20.0}, DefaultTint:=MakeColorFromHex("#7de2FFFF")}

            #Definition for the Fourth level row
            LevelRow4:=overlay:
                Slots:=array:
                    overlay_slot:
                        #Makeshift size box
                        VerticalAlignment:=vertical_alignment.Center
                        HorizontalAlignment:=horizontal_alignment.Center
                        Padding:=margin{Left:=0.0, Top:=0.0, Right:=0.0, Bottom:=0.0}
                        Widget:=texture_block{DefaultImage:=Textures.General.T_GenSquTrans, DefaultDesiredSize:=vector2{X:=1250.0, Y:=114.0}, DefaultTint:=MakeColorFromHex("#0000FFFF")}

                    overlay_slot:
                        #Rarity level content
                        VerticalAlignment:=vertical_alignment.Center
                        HorizontalAlignment:=horizontal_alignment.Left
                        Padding:=margin{Left:=0.0, Top:=0.0, Right:=0.0, Bottom:=0.0}
                        Widget:=overlay:
                            Slots:=array:

                                overlay_slot:
                                    #Makeshift size box
                                    VerticalAlignment:=vertical_alignment.Center
                                    HorizontalAlignment:=horizontal_alignment.Center
                                    Padding:=margin{Left:=0.0, Top:=0.0, Right:=0.0, Bottom:=0.0}
                                    Widget:=texture_block{DefaultImage:=Textures.General.T_GenericSquare, DefaultDesiredSize:=vector2{X:=1190.0, Y:=114.0}, DefaultTint:=MakeColorFromHex("#010a1832")}

                                overlay_slot:
                                    #Highlight texture, over the content but under the rarity flag
                                    VerticalAlignment:=vertical_alignment.Center
                                    HorizontalAlignment:=horizontal_alignment.Center
                                    Padding:=margin{Left:=0.0, Top:=0.0, Right:=0.0, Bottom:=0.0}
                                    Widget:=texture_block{DefaultImage:=Textures.General.T_LevelBorder, DefaultDesiredSize:=vector2{X:=1190.0, Y:=114.0}, DefaultTint:=MakeColorFromHex("#174583FF")}

                                    #Horizontal stack box
                                overlay_slot:
                                    VerticalAlignment:=vertical_alignment.Fill
                                    HorizontalAlignment:=horizontal_alignment.Left
                                    Padding:=margin{Left:=0.0, Top:=0.0, Right:=0.0, Bottom:=0.0}
                                    Widget:=stack_box:
                                        Orientation:=orientation.Horizontal
                                        Slots:=array:
                                            #Level flag
                                            stack_box_slot:
                                                VerticalAlignment:=vertical_alignment.Top
                                                HorizontalAlignment:=horizontal_alignment.Left
                                                Padding:=margin{Left:=0.0, Top:=0.0, Right:=15.0, Bottom:=0.0}
                                                Widget:=overlay:

                                                    Slots:=array:
                                                        overlay_slot:
                                                            VerticalAlignment:=vertical_alignment.Center
                                                            HorizontalAlignment:=horizontal_alignment.Center
                                                            Padding:=margin{Left:=0.0, Top:=0.0, Right:=0.0, Bottom:=0.0}
                                                            #Tier level makeshift size box
                                                            Widget:=texture_block{DefaultImage:=Textures.General.T_GenSquTrans, DefaultDesiredSize:=vector2{X:=150.0, Y:=114.0}, DefaultTint:=MakeColorFromHex("#FF0000FF")}

                                                        overlay_slot:
                                                            VerticalAlignment:=vertical_alignment.Top
                                                            HorizontalAlignment:=horizontal_alignment.Left
                                                            Padding:=margin{Left:=0.0, Top:=0.0, Right:=0.0, Bottom:=0.0}
                                                            #Selected level yellow flag
                                                            Widget:=overlay:

                                                                Slots:=array:
                                                                    #overlay_slot:
                                                                        #VerticalAlignment:=vertical_alignment.Center
                                                                        #HorizontalAlignment:=horizontal_alignment.Center
                                                                        #Padding:=margin{Left:=0.0, Top:=0.0, Right:=0.0, Bottom:=0.0}
                                                                        #level flag texture
                                                                        #Widget:=texture_block{DefaultImage:=Textures.General.T_GenLevelFlag, DefaultDesiredSize:=vector2{X:=140.0, Y:=45.0}, DefaultTint:=MakeColorFromHex("#f7ff1aFF")}

                                                                    #overlay_slot:
                                                                        #VerticalAlignment:=vertical_alignment.Center
                                                                        #HorizontalAlignment:=horizontal_alignment.Center
                                                                        #Padding:=margin{Left:=0.0, Top:=0.0, Right:=0.0, Bottom:=0.0}
                                                                        #level flag text
                                                                        #Widget:=text_block{DefaultText := StringToMessage("Level 0"), DefaultTextColor:=MakeColorFromHex("#000000FF")}

                                                        overlay_slot:
                                                            VerticalAlignment:=vertical_alignment.Fill
                                                            HorizontalAlignment:=horizontal_alignment.Center
                                                            Padding:=margin{Left:=0.0, Top:=5.0, Right:=0.0, Bottom:=5.0}
                                                            #Unselected level and lock
                                                            Widget:=overlay:
                                                                Slots:=array:
                                                                    
                                                                    overlay_slot:
                                                                        VerticalAlignment:=vertical_alignment.Top
                                                                        HorizontalAlignment:=horizontal_alignment.Center
                                                                        Padding:=margin{Left:=0.0, Top:=0.0, Right:=0.0, Bottom:=0.0}
                                                                        #level flag texture
                                                                        Widget:=text_block{DefaultText := StringToMessage("Level 4"), DefaultTextColor:=MakeColorFromHex("#a3a6aaff")}

                                                                    overlay_slot:
                                                                        VerticalAlignment:=vertical_alignment.Bottom
                                                                        HorizontalAlignment:=horizontal_alignment.Center
                                                                        Padding:=margin{Left:=0.0, Top:=0.0, Right:=0.0, Bottom:=0.0}
                                                                        #level flag texture
                                                                        Widget:=texture_block{DefaultImage:=Textures.General.T_Lock, DefaultDesiredSize:=vector2{X:=52.0, Y:=58.0}, DefaultTint:=MakeColorFromHex("#a3a6aaff")}
                                            
                                            #Level weapon roster
                                            stack_box_slot:
                                                VerticalAlignment:=vertical_alignment.Center
                                                HorizontalAlignment:=horizontal_alignment.Center
                                                Padding:=margin{Left:=0.0, Top:=0.0, Right:=8.0, Bottom:=0.0}
                                                Widget:=overlay:
                                                    Slots:=array:
                                                        #Weapon Icon - Pistols
                                                        overlay_slot:
                                                            VerticalAlignment:=vertical_alignment.Center
                                                            HorizontalAlignment:=horizontal_alignment.Center
                                                            Padding:=margin{Left:=0.0, Top:=0.0, Right:=0.0, Bottom:=0.0}
                                                            #Weapon icon BG
                                                            Widget:=texture_block{DefaultImage:=Textures.General.T_GenericSquare, DefaultDesiredSize:=vector2{X:=78.0, Y:=78.0}, DefaultTint:=MakeColorFromHex("#000c1d88")}

                                                        overlay_slot:
                                                            VerticalAlignment:=vertical_alignment.Center
                                                            HorizontalAlignment:=horizontal_alignment.Center
                                                            Padding:=margin{Left:=0.0, Top:=0.0, Right:=0.0, Bottom:=0.0}
                                                            #Weapon icon Border
                                                            Widget:=texture_block{DefaultImage:=Textures.General.T_GenericBorderSquare, DefaultDesiredSize:=vector2{X:=78.0, Y:=78.0}, DefaultTint:=MakeColorFromHex("#174583FF")}

                                                        overlay_slot:
                                                            VerticalAlignment:=vertical_alignment.Top
                                                            HorizontalAlignment:=horizontal_alignment.Center
                                                            Padding:=margin{Left:=0.0, Top:=5.0, Right:=0.0, Bottom:=0.0}
                                                            #Weapon icon - pistol
                                                            Widget:=texture_block{DefaultImage:=Textures.Items.T_IconPistol, DefaultDesiredSize:=vector2{X:=50.0, Y:=50.0}, DefaultTint:=MakeColorFromHex("#a3a6aaff")}

                                                        overlay_slot:
                                                            VerticalAlignment:=vertical_alignment.Bottom
                                                            HorizontalAlignment:=horizontal_alignment.Center
                                                            Padding:=margin{Left:=0.0, Top:=0.0, Right:=0.0, Bottom:=5.0}
                                                            #Weapon header - pistol
                                                            Widget:=texture_block{DefaultImage:=Textures.Items.T_HeaderPistol, DefaultDesiredSize:=vector2{X:=78.0, Y:=18.0}, DefaultTint:=MakeColorFromHex("#a3a6aaff")}
                                                        
                                            stack_box_slot:
                                                VerticalAlignment:=vertical_alignment.Center
                                                HorizontalAlignment:=horizontal_alignment.Center
                                                Padding:=margin{Left:=0.0, Top:=0.0, Right:=8.0, Bottom:=0.0}
                                                Widget:=overlay:
                                                    Slots:=array:
                                                        #Weapon Icon - Shotgun
                                                        overlay_slot:
                                                            VerticalAlignment:=vertical_alignment.Center
                                                            HorizontalAlignment:=horizontal_alignment.Center
                                                            Padding:=margin{Left:=0.0, Top:=0.0, Right:=0.0, Bottom:=0.0}
                                                            #Weapon icon BG
                                                            Widget:=texture_block{DefaultImage:=Textures.General.T_GenericSquare, DefaultDesiredSize:=vector2{X:=78.0, Y:=78.0}, DefaultTint:=MakeColorFromHex("#000c1d88")}

                                                        overlay_slot:
                                                            VerticalAlignment:=vertical_alignment.Center
                                                            HorizontalAlignment:=horizontal_alignment.Center
                                                            Padding:=margin{Left:=0.0, Top:=0.0, Right:=0.0, Bottom:=0.0}
                                                            #Weapon icon Border
                                                            Widget:=texture_block{DefaultImage:=Textures.General.T_GenericBorderSquare, DefaultDesiredSize:=vector2{X:=78.0, Y:=78.0}, DefaultTint:=MakeColorFromHex("#174583FF")}

                                                        overlay_slot:
                                                            VerticalAlignment:=vertical_alignment.Top
                                                            HorizontalAlignment:=horizontal_alignment.Center
                                                            Padding:=margin{Left:=0.0, Top:=5.0, Right:=0.0, Bottom:=0.0}
                                                            #Weapon icon - shotgun
                                                            Widget:=texture_block{DefaultImage:=Textures.Items.T_IconShotgun, DefaultDesiredSize:=vector2{X:=50.0, Y:=50.0}, DefaultTint:=MakeColorFromHex("#a3a6aaff")}

                                                        overlay_slot:
                                                            VerticalAlignment:=vertical_alignment.Bottom
                                                            HorizontalAlignment:=horizontal_alignment.Center
                                                            Padding:=margin{Left:=0.0, Top:=0.0, Right:=0.0, Bottom:=5.0}
                                                            #Weapon header - shotgun
                                                            Widget:=texture_block{DefaultImage:=Textures.Items.T_HeaderShotgun, DefaultDesiredSize:=vector2{X:=78.0, Y:=18.0}, DefaultTint:=MakeColorFromHex("#a3a6aaff")}

                                            stack_box_slot:
                                                VerticalAlignment:=vertical_alignment.Center
                                                HorizontalAlignment:=horizontal_alignment.Center
                                                Padding:=margin{Left:=0.0, Top:=0.0, Right:=8.0, Bottom:=0.0}
                                                Widget:=overlay:
                                                    Slots:=array:
                                                        #Weapon Icon - SMG
                                                        overlay_slot:
                                                            VerticalAlignment:=vertical_alignment.Top
                                                            HorizontalAlignment:=horizontal_alignment.Center
                                                            Padding:=margin{Left:=0.0, Top:=0.0, Right:=0.0, Bottom:=0.0}
                                                            #Weapon icon BG
                                                            Widget:=texture_block{DefaultImage:=Textures.General.T_GenericSquare, DefaultDesiredSize:=vector2{X:=78.0, Y:=78.0}, DefaultTint:=MakeColorFromHex("#000c1d88")}

                                                        overlay_slot:
                                                            VerticalAlignment:=vertical_alignment.Center
                                                            HorizontalAlignment:=horizontal_alignment.Center
                                                            Padding:=margin{Left:=0.0, Top:=0.0, Right:=0.0, Bottom:=0.0}
                                                            #Weapon icon Border
                                                            Widget:=texture_block{DefaultImage:=Textures.General.T_GenericBorderSquare, DefaultDesiredSize:=vector2{X:=78.0, Y:=78.0}, DefaultTint:=MakeColorFromHex("#174583FF")}

                                                        overlay_slot:
                                                            VerticalAlignment:=vertical_alignment.Top
                                                            HorizontalAlignment:=horizontal_alignment.Center
                                                            Padding:=margin{Left:=0.0, Top:=5.0, Right:=0.0, Bottom:=0.0}
                                                            #Weapon icon - SMG
                                                            Widget:=texture_block{DefaultImage:=Textures.Items.T_IconSMG, DefaultDesiredSize:=vector2{X:=50.0, Y:=50.0}, DefaultTint:=MakeColorFromHex("#a3a6aaff")}

                                                        overlay_slot:
                                                            VerticalAlignment:=vertical_alignment.Bottom
                                                            HorizontalAlignment:=horizontal_alignment.Center
                                                            Padding:=margin{Left:=0.0, Top:=0.0, Right:=0.0, Bottom:=5.0}
                                                            #Weapon header - SMG
                                                            Widget:=texture_block{DefaultImage:=Textures.Items.T_HeaderSMGs, DefaultDesiredSize:=vector2{X:=78.0, Y:=18.0}, DefaultTint:=MakeColorFromHex("#a3a6aaff")}
                                                            

                                            stack_box_slot:
                                                VerticalAlignment:=vertical_alignment.Center
                                                HorizontalAlignment:=horizontal_alignment.Center
                                                Padding:=margin{Left:=0.0, Top:=0.0, Right:=8.0, Bottom:=0.0}
                                                Widget:=overlay:
                                                    Slots:=array:
                                                        #Weapon Icon - Rifle
                                                        overlay_slot:
                                                            VerticalAlignment:=vertical_alignment.Center
                                                            HorizontalAlignment:=horizontal_alignment.Center
                                                            Padding:=margin{Left:=0.0, Top:=0.0, Right:=0.0, Bottom:=0.0}
                                                            #Weapon icon BG
                                                            Widget:=texture_block{DefaultImage:=Textures.General.T_GenericSquare, DefaultDesiredSize:=vector2{X:=78.0, Y:=78.0}, DefaultTint:=MakeColorFromHex("#000c1d88")}

                                                        overlay_slot:
                                                            VerticalAlignment:=vertical_alignment.Top
                                                            HorizontalAlignment:=horizontal_alignment.Center
                                                            Padding:=margin{Left:=0.0, Top:=0.0, Right:=0.0, Bottom:=0.0}
                                                            #Weapon icon Border
                                                            Widget:=texture_block{DefaultImage:=Textures.General.T_GenericBorderSquare, DefaultDesiredSize:=vector2{X:=78.0, Y:=78.0}, DefaultTint:=MakeColorFromHex("#174583FF")}

                                                        overlay_slot:
                                                            VerticalAlignment:=vertical_alignment.Top
                                                            HorizontalAlignment:=horizontal_alignment.Center
                                                            Padding:=margin{Left:=0.0, Top:=5.0, Right:=0.0, Bottom:=0.0}
                                                            #Weapon icon - Rifle
                                                            Widget:=texture_block{DefaultImage:=Textures.Items.T_IconRifle, DefaultDesiredSize:=vector2{X:=50.0, Y:=50.0}, DefaultTint:=MakeColorFromHex("#a3a6aaff")}

                                                        overlay_slot:
                                                            VerticalAlignment:=vertical_alignment.Bottom
                                                            HorizontalAlignment:=horizontal_alignment.Center
                                                            Padding:=margin{Left:=0.0, Top:=0.0, Right:=0.0, Bottom:=5.0}
                                                            #Weapon header - Rifle
                                                            Widget:=texture_block{DefaultImage:=Textures.Items.T_HeaderRifle, DefaultDesiredSize:=vector2{X:=78.0, Y:=18.0}, DefaultTint:=MakeColorFromHex("#a3a6aaff")}

                                            stack_box_slot:
                                                VerticalAlignment:=vertical_alignment.Center
                                                HorizontalAlignment:=horizontal_alignment.Center
                                                Padding:=margin{Left:=0.0, Top:=0.0, Right:=8.0, Bottom:=0.0}
                                                Widget:=overlay:
                                                    Slots:=array:
                                                        #Weapon Icon - Explosive
                                                        overlay_slot:
                                                            VerticalAlignment:=vertical_alignment.Center
                                                            HorizontalAlignment:=horizontal_alignment.Center
                                                            Padding:=margin{Left:=0.0, Top:=0.0, Right:=0.0, Bottom:=0.0}
                                                            #Weapon icon BG
                                                            Widget:=texture_block{DefaultImage:=Textures.General.T_GenericSquare, DefaultDesiredSize:=vector2{X:=78.0, Y:=78.0}, DefaultTint:=MakeColorFromHex("#000c1d88")}

                                                        overlay_slot:
                                                            VerticalAlignment:=vertical_alignment.Center
                                                            HorizontalAlignment:=horizontal_alignment.Center
                                                            Padding:=margin{Left:=0.0, Top:=0.0, Right:=0.0, Bottom:=0.0}
                                                            #Weapon icon Border
                                                            Widget:=texture_block{DefaultImage:=Textures.General.T_GenericBorderSquare, DefaultDesiredSize:=vector2{X:=78.0, Y:=78.0}, DefaultTint:=MakeColorFromHex("#174583FF")}

                                                        overlay_slot:
                                                            VerticalAlignment:=vertical_alignment.Top
                                                            HorizontalAlignment:=horizontal_alignment.Center
                                                            Padding:=margin{Left:=0.0, Top:=5.0, Right:=0.0, Bottom:=0.0}
                                                            #Weapon icon - Explosive
                                                            Widget:=texture_block{DefaultImage:=Textures.Items.T_IconExplosive, DefaultDesiredSize:=vector2{X:=50.0, Y:=50.0}, DefaultTint:=MakeColorFromHex("#a3a6aaff")}

                                                        overlay_slot:
                                                            VerticalAlignment:=vertical_alignment.Bottom
                                                            HorizontalAlignment:=horizontal_alignment.Center
                                                            Padding:=margin{Left:=0.0, Top:=0.0, Right:=0.0, Bottom:=5.0}
                                                            #Weapon header - Explosive
                                                            Widget:=texture_block{DefaultImage:=Textures.Items.T_HeaderExplosive, DefaultDesiredSize:=vector2{X:=78.0, Y:=18.0}, DefaultTint:=MakeColorFromHex("#a3a6aaff")}

                                            stack_box_slot:
                                                VerticalAlignment:=vertical_alignment.Center
                                                HorizontalAlignment:=horizontal_alignment.Center
                                                Padding:=margin{Left:=0.0, Top:=0.0, Right:=8.0, Bottom:=0.0}
                                                Widget:=overlay:
                                                    Slots:=array:
                                                        #Weapon Icon - Special
                                                        overlay_slot:
                                                            VerticalAlignment:=vertical_alignment.Center
                                                            HorizontalAlignment:=horizontal_alignment.Center
                                                            Padding:=margin{Left:=0.0, Top:=0.0, Right:=0.0, Bottom:=0.0}
                                                            #Weapon icon BG
                                                            Widget:=texture_block{DefaultImage:=Textures.General.T_GenericSquare, DefaultDesiredSize:=vector2{X:=78.0, Y:=78.0}, DefaultTint:=MakeColorFromHex("#000c1d88")}

                                                        overlay_slot:
                                                            VerticalAlignment:=vertical_alignment.Center
                                                            HorizontalAlignment:=horizontal_alignment.Center
                                                            Padding:=margin{Left:=0.0, Top:=0.0, Right:=0.0, Bottom:=0.0}
                                                            #Weapon icon Border
                                                            Widget:=texture_block{DefaultImage:=Textures.General.T_GenericBorderSquare, DefaultDesiredSize:=vector2{X:=78.0, Y:=78.0}, DefaultTint:=MakeColorFromHex("#174583FF")}

                                                        overlay_slot:
                                                            VerticalAlignment:=vertical_alignment.Top
                                                            HorizontalAlignment:=horizontal_alignment.Center
                                                            Padding:=margin{Left:=0.0, Top:=5.0, Right:=0.0, Bottom:=0.0}
                                                            #Weapon icon - special
                                                            Widget:=texture_block{DefaultImage:=Textures.Items.T_IconArtifact, DefaultDesiredSize:=vector2{X:=50.0, Y:=50.0}, DefaultTint:=MakeColorFromHex("#a3a6aaff")}

                                                        overlay_slot:
                                                            VerticalAlignment:=vertical_alignment.Bottom
                                                            HorizontalAlignment:=horizontal_alignment.Center
                                                            Padding:=margin{Left:=0.0, Top:=0.0, Right:=0.0, Bottom:=5.0}
                                                            #Weapon header - Special
                                                            Widget:=texture_block{DefaultImage:=Textures.Items.T_HeaderSpecial, DefaultDesiredSize:=vector2{X:=78.0, Y:=18.0}, DefaultTint:=MakeColorFromHex("#a3a6aaff")}

                                            stack_box_slot:
                                                VerticalAlignment:=vertical_alignment.Center
                                                HorizontalAlignment:=horizontal_alignment.Center
                                                Padding:=margin{Left:=0.0, Top:=0.0, Right:=8.0, Bottom:=0.0}
                                                Widget:=overlay:
                                                    Slots:=array:
                                                        #Weapon Icon - DMR Sniper
                                                        overlay_slot:
                                                            VerticalAlignment:=vertical_alignment.Center
                                                            HorizontalAlignment:=horizontal_alignment.Center
                                                            Padding:=margin{Left:=0.0, Top:=0.0, Right:=0.0, Bottom:=0.0}
                                                            #Weapon icon BG
                                                            Widget:=texture_block{DefaultImage:=Textures.General.T_GenericSquare, DefaultDesiredSize:=vector2{X:=78.0, Y:=78.0}, DefaultTint:=MakeColorFromHex("#000c1d88")}

                                                        overlay_slot:
                                                            VerticalAlignment:=vertical_alignment.Center
                                                            HorizontalAlignment:=horizontal_alignment.Center
                                                            Padding:=margin{Left:=0.0, Top:=0.0, Right:=0.0, Bottom:=0.0}
                                                            #Weapon icon Border
                                                            Widget:=texture_block{DefaultImage:=Textures.General.T_GenericBorderSquare, DefaultDesiredSize:=vector2{X:=78.0, Y:=78.0}, DefaultTint:=MakeColorFromHex("#174583FF")}

                                                        overlay_slot:
                                                            VerticalAlignment:=vertical_alignment.Top
                                                            HorizontalAlignment:=horizontal_alignment.Center
                                                            Padding:=margin{Left:=0.0, Top:=5.0, Right:=0.0, Bottom:=0.0}
                                                            #Weapon icon - Sniper
                                                            Widget:=texture_block{DefaultImage:=Textures.Items.T_IconSniper, DefaultDesiredSize:=vector2{X:=50.0, Y:=50.0}, DefaultTint:=MakeColorFromHex("#a3a6aaff")}

                                                        overlay_slot:
                                                            VerticalAlignment:=vertical_alignment.Bottom
                                                            HorizontalAlignment:=horizontal_alignment.Center
                                                            Padding:=margin{Left:=0.0, Top:=0.0, Right:=0.0, Bottom:=5.0}
                                                            #Weapon header - DMR Sniper
                                                            Widget:=texture_block{DefaultImage:=Textures.Items.T_HeaderDMRSniper, DefaultDesiredSize:=vector2{X:=78.0, Y:=18.0}, DefaultTint:=MakeColorFromHex("#a3a6aaff")}

                                            stack_box_slot:
                                                VerticalAlignment:=vertical_alignment.Fill
                                                HorizontalAlignment:=horizontal_alignment.Center #Center
                                                Padding:=margin{Left:=0.0, Top:=0.0, Right:=0.0, Bottom:=0.0}
                                                Widget:=overlay:
                                                    Slots:=array:
                                                        overlay_slot:
                                                            #Makeshift size box
                                                            VerticalAlignment:=vertical_alignment.Center
                                                            HorizontalAlignment:=horizontal_alignment.Center
                                                            Padding:=margin{Left:=0.0, Top:=0.0, Right:=0.0, Bottom:=0.0}
                                                            Widget:=texture_block{DefaultImage:=Textures.General.T_GenSquTrans, DefaultDesiredSize:=vector2{X:=418.0, Y:=100.0}, DefaultTint:=MakeColorFromHex("#00FF0011")}

                                                        overlay_slot:
                                                            #Vertical stack box to add all the rarity information
                                                            VerticalAlignment:=vertical_alignment.Center
                                                            HorizontalAlignment:=horizontal_alignment.Center
                                                            Padding:=margin{Left:=0.0, Top:=0.0, Right:=0.0, Bottom:=0.0}
                                                            Widget:=stack_box:
                                                                Orientation:=orientation.Vertical
                                                                Slots:=array:
                                                                    stack_box_slot:
                                                                        #Horizontal Stack box for max rarity available
                                                                        VerticalAlignment:=vertical_alignment.Top
                                                                        HorizontalAlignment:=horizontal_alignment.Center
                                                                        Padding:=margin{Left:=0.0, Top:=0.0, Right:=0.0, Bottom:=0.0}
                                                                        Widget:=stack_box:
                                                                            Orientation:=orientation.Horizontal
                                                                            Slots:=array:
                                                                                stack_box_slot:
                                                                                    VerticalAlignment:=vertical_alignment.Center
                                                                                    HorizontalAlignment:=horizontal_alignment.Center
                                                                                    Padding:=margin{Left:=0.0, Top:=0.0, Right:=0.0, Bottom:=0.0}
                                                                                    Widget:=text_block{DefaultText := StringToMessage("Rarity chance up to "), DefaultTextColor:=MakeColorFromHex("#a3a6aaff")}

                                                                                stack_box_slot:
                                                                                    VerticalAlignment:=vertical_alignment.Center
                                                                                    HorizontalAlignment:=horizontal_alignment.Center
                                                                                    Padding:=margin{Left:=0.0, Top:=0.0, Right:=0.0, Bottom:=0.0}
                                                                                    Widget:=texture_block{DefaultImage:=Textures.Rarity.T_FlagLegendary, DefaultDesiredSize:=vector2{X:=104.0, Y:=36.0}, DefaultTint:=MakeColorFromHex("#a3a6aaff")}

                                                                    stack_box_slot:
                                                                        #Horizontal overlay for overall rarity scheme
                                                                        VerticalAlignment:=vertical_alignment.Bottom
                                                                        HorizontalAlignment:=horizontal_alignment.Center
                                                                        Padding:=margin{Left:=0.0, Top:=0.0, Right:=0.0, Bottom:=0.0}
                                                                        Widget:=overlay:
                                                                            Slots:=array:
                                                                                
                                                                                #Makeshift size box for the rarity scheme
                                                                                overlay_slot:
                                                                                    VerticalAlignment:=vertical_alignment.Center
                                                                                    HorizontalAlignment:=horizontal_alignment.Center
                                                                                    Padding:=margin{Left:=0.0, Top:=0.0, Right:=0.0, Bottom:=0.0}
                                                                                    Widget:=texture_block{DefaultImage:=Textures.General.T_GenSquTrans, DefaultDesiredSize:=vector2{X:=255.0, Y:=50.0}, DefaultTint:=MakeColorFromHex("#FF000022")}#Red frame

                                                                                #Visible border for rarity layout
                                                                                overlay_slot:
                                                                                    VerticalAlignment:=vertical_alignment.Center
                                                                                    HorizontalAlignment:=horizontal_alignment.Center
                                                                                    Padding:=margin{Left:=0.0, Top:=0.0, Right:=0.0, Bottom:=0.0}
                                                                                    Widget:=texture_block{DefaultImage:=Textures.General.T_GenericBorderSquare3, DefaultDesiredSize:=vector2{X:=255.0, Y:=40.0}, DefaultTint:=MakeColorFromHex("#174583FF")}

                                                                                #Horizontal Stack box for rarity slots
                                                                                overlay_slot:
                                                                                    VerticalAlignment:=vertical_alignment.Fill
                                                                                    HorizontalAlignment:=horizontal_alignment.Fill
                                                                                    Padding:=margin{Left:=5.0, Top:=0.0, Right:=5.0, Bottom:=0.0}
                                                                                    Widget:=stack_box:
                                                                                        Orientation:=orientation.Horizontal
                                                                                        Slots:=array:
                                                                                            #Overlay for each rarity slots, including the pointing arrow
                                                                                            stack_box_slot:
                                                                                                VerticalAlignment:=vertical_alignment.Center
                                                                                                HorizontalAlignment:=horizontal_alignment.Center
                                                                                                Distribution:=option{1.0}
                                                                                                Padding:=margin{Left:=0.0, Top:=0.0, Right:=0.0, Bottom:=0.0}
                                                                                                Widget:=overlay:
                                                                                                    Slots:=array:
                                                                                                        #Makeshift size box
                                                                                                        overlay_slot:
                                                                                                            VerticalAlignment:=vertical_alignment.Center
                                                                                                            HorizontalAlignment:=horizontal_alignment.Center
                                                                                                            Padding:=margin{Left:=0.0, Top:=0.0, Right:=0.0, Bottom:=0.0}
                                                                                                            Widget:=texture_block{DefaultImage:=Textures.General.T_GenSquTrans, DefaultDesiredSize:=vector2{X:=30.0, Y:=40.0}, DefaultTint:=MakeColorFromHex("#0000FF08")}

                                                                                                        #Overlay for rarity border and texture
                                                                                                        overlay_slot:
                                                                                                            VerticalAlignment:=vertical_alignment.Center
                                                                                                            HorizontalAlignment:=horizontal_alignment.Center
                                                                                                            Padding:=margin{Left:=0.0, Top:=0.0, Right:=0.0, Bottom:=0.0}
                                                                                                            Widget:=overlay:
                                                                                                                Slots:=array:
                                                                                                                    
                                                                                                                    #Rarity texture border, when the rarity slot is empty
                                                                                                                    overlay_slot:
                                                                                                                        VerticalAlignment:=vertical_alignment.Center
                                                                                                                        HorizontalAlignment:=horizontal_alignment.Center
                                                                                                                        Padding:=margin{Left:=0.0, Top:=0.0, Right:=0.0, Bottom:=0.0}
                                                                                                                        Widget:=texture_block{DefaultImage:=Textures.General.T_GenericBorderSquare2, DefaultDesiredSize:=vector2{X:=30.0, Y:=30.0}, DefaultTint:=MakeColorFromHex("#174583FF")}

                                                                                                                    #Rarity texture
                                                                                                                    #overlay_slot:
                                                                                                                        #VerticalAlignment:=vertical_alignment.Center
                                                                                                                        #HorizontalAlignment:=horizontal_alignment.Center
                                                                                                                        #Padding:=margin{Left:=0.0, Top:=0.0, Right:=0.0, Bottom:=0.0}
                                                                                                                        #Widget:=texture_block{DefaultImage:=Textures.General.T_GenericSquare, DefaultDesiredSize:=vector2{X:=30.0, Y:=30.0}, DefaultTint:=MakeColorFromHex("#979797FF")}
                                                                                                        
                                                                                                                    

                                                                                                        #Bracket
                                                                                                        overlay_slot:
                                                                                                            VerticalAlignment:=vertical_alignment.Fill
                                                                                                            HorizontalAlignment:=horizontal_alignment.Fill
                                                                                                            Padding:=margin{Left:=0.0, Top:=0.0, Right:=0.0, Bottom:=0.0}
                                                                                                            Widget:=stack_box:
                                                                                                                Orientation:=orientation.Horizontal
                                                                                                                #two instances of bracket texture to represent the range of rarities
                                                                                                                Slots:=array:

                                                                                                                    #stack_box_slot:
                                                                                                                        #VerticalAlignment:=vertical_alignment.Center
                                                                                                                        #HorizontalAlignment:=horizontal_alignment.Left
                                                                                                                        #Distribution:=option{1.0}
                                                                                                                        #Padding:=margin{Left:=0.0, Top:=0.0, Right:=0.0, Bottom:=0.0}
                                                                                                                        #Widget:=texture_block{DefaultImage:=Textures.General.T_RarityBracketLeft, DefaultDesiredSize:=vector2{X:=18.0, Y:=36.0}, DefaultTint:=MakeColorFromHex("#FFFFFF44")}

                                                                                                                    #stack_box_slot:
                                                                                                                        #VerticalAlignment:=vertical_alignment.Center
                                                                                                                        #HorizontalAlignment:=horizontal_alignment.Right
                                                                                                                        #Distribution:=option{1.0}
                                                                                                                        #Padding:=margin{Left:=0.0, Top:=0.0, Right:=0.0, Bottom:=0.0}
                                                                                                                        #Widget:=texture_block{DefaultImage:=Textures.General.T_RarityBracketRight, DefaultDesiredSize:=vector2{X:=18.0, Y:=36.0}, DefaultTint:=MakeColorFromHex("#FFFFFF44")}

                                                                                            stack_box_slot:
                                                                                                VerticalAlignment:=vertical_alignment.Center
                                                                                                HorizontalAlignment:=horizontal_alignment.Center
                                                                                                Distribution:=option{1.0}
                                                                                                Padding:=margin{Left:=0.0, Top:=0.0, Right:=0.0, Bottom:=0.0}
                                                                                                Widget:=overlay:
                                                                                                    Slots:=array:
                                                                                                        #Makeshift size box
                                                                                                        overlay_slot:
                                                                                                            VerticalAlignment:=vertical_alignment.Center
                                                                                                            HorizontalAlignment:=horizontal_alignment.Center
                                                                                                            Padding:=margin{Left:=0.0, Top:=0.0, Right:=0.0, Bottom:=0.0}
                                                                                                            Widget:=texture_block{DefaultImage:=Textures.General.T_GenSquTrans, DefaultDesiredSize:=vector2{X:=30.0, Y:=40.0}, DefaultTint:=MakeColorFromHex("#0000FF08")}

                                                                                                        #Overlay for rarity border and texture
                                                                                                        overlay_slot:
                                                                                                            VerticalAlignment:=vertical_alignment.Center
                                                                                                            HorizontalAlignment:=horizontal_alignment.Center
                                                                                                            Padding:=margin{Left:=0.0, Top:=0.0, Right:=0.0, Bottom:=0.0}
                                                                                                            Widget:=overlay:
                                                                                                                Slots:=array:
                                                                                                                    
                                                                                                                    #Rarity texture border, when the rarity slot is empty
                                                                                                                    #overlay_slot:
                                                                                                                        #VerticalAlignment:=vertical_alignment.Center
                                                                                                                        #HorizontalAlignment:=horizontal_alignment.Center
                                                                                                                        #Padding:=margin{Left:=0.0, Top:=0.0, Right:=0.0, Bottom:=0.0}
                                                                                                                        #Widget:=texture_block{DefaultImage:=Textures.General.T_GenericBorderSquare2, DefaultDesiredSize:=vector2{X:=30.0, Y:=30.0}, DefaultTint:=MakeColorFromHex("#2e7be388")}

                                                                                                                    #Rarity texture
                                                                                                                    overlay_slot:
                                                                                                                        VerticalAlignment:=vertical_alignment.Center
                                                                                                                        HorizontalAlignment:=horizontal_alignment.Center
                                                                                                                        Padding:=margin{Left:=0.0, Top:=0.0, Right:=0.0, Bottom:=0.0}
                                                                                                                        Widget:=texture_block{DefaultImage:=Textures.General.T_GenericSquare, DefaultDesiredSize:=vector2{X:=20.0, Y:=20.0}, DefaultTint:=MakeColorFromHex("#2a7028FF")}

                                                                                                        #Bracket
                                                                                                        overlay_slot:
                                                                                                            VerticalAlignment:=vertical_alignment.Fill
                                                                                                            HorizontalAlignment:=horizontal_alignment.Fill
                                                                                                            Padding:=margin{Left:=0.0, Top:=0.0, Right:=0.0, Bottom:=0.0}
                                                                                                            Widget:=stack_box:
                                                                                                                Orientation:=orientation.Horizontal
                                                                                                                #two instances of bracket texture to represent the range of rarities
                                                                                                                Slots:=array:

                                                                                                                    stack_box_slot:
                                                                                                                        VerticalAlignment:=vertical_alignment.Center
                                                                                                                        HorizontalAlignment:=horizontal_alignment.Left
                                                                                                                        Distribution:=option{1.0}
                                                                                                                        Padding:=margin{Left:=0.0, Top:=0.0, Right:=0.0, Bottom:=0.0}
                                                                                                                        Widget:=texture_block{DefaultImage:=Textures.General.T_RarityBracketLeft, DefaultDesiredSize:=vector2{X:=18.0, Y:=36.0}, DefaultTint:=MakeColorFromHex("#a3a6aaff")}

                                                                                                                    #stack_box_slot:
                                                                                                                        #VerticalAlignment:=vertical_alignment.Center
                                                                                                                        #HorizontalAlignment:=horizontal_alignment.Right
                                                                                                                        #Distribution:=option{1.0}
                                                                                                                        #Padding:=margin{Left:=0.0, Top:=0.0, Right:=0.0, Bottom:=0.0}
                                                                                                                        #Widget:=texture_block{DefaultImage:=Textures.General.T_RarityBracketRight, DefaultDesiredSize:=vector2{X:=18.0, Y:=36.0}, DefaultTint:=MakeColorFromHex("#FFFFFF44")}

                                                                                            stack_box_slot:
                                                                                                VerticalAlignment:=vertical_alignment.Center
                                                                                                HorizontalAlignment:=horizontal_alignment.Center
                                                                                                Distribution:=option{1.0}
                                                                                                Padding:=margin{Left:=0.0, Top:=0.0, Right:=0.0, Bottom:=0.0}
                                                                                                Widget:=overlay:
                                                                                                    Slots:=array:
                                                                                                        #Makeshift size box
                                                                                                        overlay_slot:
                                                                                                            VerticalAlignment:=vertical_alignment.Center
                                                                                                            HorizontalAlignment:=horizontal_alignment.Center
                                                                                                            Padding:=margin{Left:=0.0, Top:=0.0, Right:=0.0, Bottom:=0.0}
                                                                                                            Widget:=texture_block{DefaultImage:=Textures.General.T_GenSquTrans, DefaultDesiredSize:=vector2{X:=30.0, Y:=40.0}, DefaultTint:=MakeColorFromHex("#0000FF08")}

                                                                                                        #Overlay for rarity border and texture
                                                                                                        overlay_slot:
                                                                                                            VerticalAlignment:=vertical_alignment.Center
                                                                                                            HorizontalAlignment:=horizontal_alignment.Center
                                                                                                            Padding:=margin{Left:=0.0, Top:=0.0, Right:=0.0, Bottom:=0.0}
                                                                                                            Widget:=overlay:
                                                                                                                Slots:=array:
                                                                                                                                                                                                                
                                                                                                                    #Rarity texture border, when the rarity slot is empty
                                                                                                                    #overlay_slot:
                                                                                                                        #VerticalAlignment:=vertical_alignment.Center
                                                                                                                        #HorizontalAlignment:=horizontal_alignment.Center
                                                                                                                        #Padding:=margin{Left:=0.0, Top:=0.0, Right:=0.0, Bottom:=0.0}
                                                                                                                        #Widget:=texture_block{DefaultImage:=Textures.General.T_GenericBorderSquare2, DefaultDesiredSize:=vector2{X:=30.0, Y:=30.0}, DefaultTint:=MakeColorFromHex("#2e7be388")}
                                                                                                                                                                                                                            
                                                                                                                    #Rarity texture
                                                                                                                    overlay_slot:
                                                                                                                        VerticalAlignment:=vertical_alignment.Center
                                                                                                                        HorizontalAlignment:=horizontal_alignment.Center
                                                                                                                        Padding:=margin{Left:=0.0, Top:=0.0, Right:=0.0, Bottom:=0.0}
                                                                                                                        Widget:=texture_block{DefaultImage:=Textures.General.T_GenericSquare, DefaultDesiredSize:=vector2{X:=20.0, Y:=20.0}, DefaultTint:=MakeColorFromHex("#183379FF")}
                                                                                                                    
                                                                                                        #Bracket
                                                                                                        overlay_slot:
                                                                                                            VerticalAlignment:=vertical_alignment.Fill
                                                                                                            HorizontalAlignment:=horizontal_alignment.Fill
                                                                                                            Padding:=margin{Left:=0.0, Top:=0.0, Right:=0.0, Bottom:=0.0}
                                                                                                            Widget:=stack_box:
                                                                                                                Orientation:=orientation.Horizontal
                                                                                                                #two instances of bracket texture to represent the range of rarities
                                                                                                                Slots:=array:

                                                                                                                    #stack_box_slot:
                                                                                                                        #VerticalAlignment:=vertical_alignment.Center
                                                                                                                        #HorizontalAlignment:=horizontal_alignment.Left
                                                                                                                        #Distribution:=option{1.0}
                                                                                                                        #Padding:=margin{Left:=0.0, Top:=0.0, Right:=0.0, Bottom:=0.0}
                                                                                                                        #Widget:=texture_block{DefaultImage:=Textures.General.T_RarityBracketLeft, DefaultDesiredSize:=vector2{X:=18.0, Y:=36.0}, DefaultTint:=MakeColorFromHex("#FFFFFF44")}

                                                                                                                    #stack_box_slot:
                                                                                                                        #VerticalAlignment:=vertical_alignment.Center
                                                                                                                        #HorizontalAlignment:=horizontal_alignment.Right
                                                                                                                        #Distribution:=option{1.0}
                                                                                                                        #Padding:=margin{Left:=0.0, Top:=0.0, Right:=0.0, Bottom:=0.0}
                                                                                                                        #Widget:=texture_block{DefaultImage:=Textures.General.T_RarityBracketRight, DefaultDesiredSize:=vector2{X:=18.0, Y:=36.0}, DefaultTint:=MakeColorFromHex("#FFFFFF44")}

                                                                                            stack_box_slot:
                                                                                                VerticalAlignment:=vertical_alignment.Center
                                                                                                HorizontalAlignment:=horizontal_alignment.Center
                                                                                                Distribution:=option{1.0}
                                                                                                Padding:=margin{Left:=0.0, Top:=0.0, Right:=0.0, Bottom:=0.0}
                                                                                                Widget:=overlay:
                                                                                                    Slots:=array:
                                                                                                        #Makeshift size box
                                                                                                        overlay_slot:
                                                                                                            VerticalAlignment:=vertical_alignment.Center
                                                                                                            HorizontalAlignment:=horizontal_alignment.Center
                                                                                                            Padding:=margin{Left:=0.0, Top:=0.0, Right:=0.0, Bottom:=0.0}
                                                                                                            Widget:=texture_block{DefaultImage:=Textures.General.T_GenSquTrans, DefaultDesiredSize:=vector2{X:=30.0, Y:=40.0}, DefaultTint:=MakeColorFromHex("#0000FF08")}

                                                                                                        #Overlay for rarity border and texture
                                                                                                        overlay_slot:
                                                                                                            VerticalAlignment:=vertical_alignment.Center
                                                                                                            HorizontalAlignment:=horizontal_alignment.Center
                                                                                                            Padding:=margin{Left:=0.0, Top:=0.0, Right:=0.0, Bottom:=0.0}
                                                                                                            Widget:=overlay:
                                                                                                                Slots:=array:

                                                                                                                    #Rarity texture border, when the rarity slot is empty
                                                                                                                    #overlay_slot:
                                                                                                                        #VerticalAlignment:=vertical_alignment.Center
                                                                                                                        #HorizontalAlignment:=horizontal_alignment.Center
                                                                                                                        #Padding:=margin{Left:=0.0, Top:=0.0, Right:=0.0, Bottom:=0.0}
                                                                                                                        #Widget:=texture_block{DefaultImage:=Textures.General.T_GenericBorderSquare2, DefaultDesiredSize:=vector2{X:=30.0, Y:=30.0}, DefaultTint:=MakeColorFromHex("#2e7be388")}
                                                                                                                    
                                                                                                                    #Rarity texture
                                                                                                                    overlay_slot:
                                                                                                                        VerticalAlignment:=vertical_alignment.Center
                                                                                                                        HorizontalAlignment:=horizontal_alignment.Center
                                                                                                                        Padding:=margin{Left:=0.0, Top:=0.0, Right:=0.0, Bottom:=0.0}
                                                                                                                        Widget:=texture_block{DefaultImage:=Textures.General.T_GenericSquare, DefaultDesiredSize:=vector2{X:=20.0, Y:=20.0}, DefaultTint:=MakeColorFromHex("#5b1e79FF")}
                                                                                                                    
                                                                                                        #Bracket
                                                                                                        overlay_slot:
                                                                                                            VerticalAlignment:=vertical_alignment.Fill
                                                                                                            HorizontalAlignment:=horizontal_alignment.Fill
                                                                                                            Padding:=margin{Left:=0.0, Top:=0.0, Right:=0.0, Bottom:=0.0}
                                                                                                            Widget:=stack_box:
                                                                                                                Orientation:=orientation.Horizontal
                                                                                                                #two instances of bracket texture to represent the range of rarities
                                                                                                                Slots:=array:

                                                                                                                    #stack_box_slot:
                                                                                                                        #VerticalAlignment:=vertical_alignment.Center
                                                                                                                        #HorizontalAlignment:=horizontal_alignment.Left
                                                                                                                        #Distribution:=option{1.0}
                                                                                                                        #Padding:=margin{Left:=0.0, Top:=0.0, Right:=0.0, Bottom:=0.0}
                                                                                                                        #Widget:=texture_block{DefaultImage:=Textures.General.T_RarityBracketLeft, DefaultDesiredSize:=vector2{X:=18.0, Y:=36.0}, DefaultTint:=MakeColorFromHex("#FFFFFF44")}

                                                                                                                    #stack_box_slot:
                                                                                                                        #VerticalAlignment:=vertical_alignment.Center
                                                                                                                        #HorizontalAlignment:=horizontal_alignment.Right
                                                                                                                        #Distribution:=option{1.0}
                                                                                                                        #Padding:=margin{Left:=0.0, Top:=0.0, Right:=0.0, Bottom:=0.0}
                                                                                                                        #Widget:=texture_block{DefaultImage:=Textures.General.T_RarityBracketRight, DefaultDesiredSize:=vector2{X:=18.0, Y:=36.0}, DefaultTint:=MakeColorFromHex("#FFFFFF44")}

                                                                                            stack_box_slot:
                                                                                                VerticalAlignment:=vertical_alignment.Center
                                                                                                HorizontalAlignment:=horizontal_alignment.Center
                                                                                                Distribution:=option{1.0}
                                                                                                Padding:=margin{Left:=0.0, Top:=0.0, Right:=0.0, Bottom:=0.0}
                                                                                                Widget:=overlay:
                                                                                                    Slots:=array:
                                                                                                        #Makeshift size box
                                                                                                        overlay_slot:
                                                                                                            VerticalAlignment:=vertical_alignment.Center
                                                                                                            HorizontalAlignment:=horizontal_alignment.Center
                                                                                                            Padding:=margin{Left:=0.0, Top:=0.0, Right:=0.0, Bottom:=0.0}
                                                                                                            Widget:=texture_block{DefaultImage:=Textures.General.T_GenSquTrans, DefaultDesiredSize:=vector2{X:=30.0, Y:=40.0}, DefaultTint:=MakeColorFromHex("#0000FF08")}

                                                                                                        #Overlay for rarity border and texture
                                                                                                        overlay_slot:
                                                                                                            VerticalAlignment:=vertical_alignment.Center
                                                                                                            HorizontalAlignment:=horizontal_alignment.Center
                                                                                                            Padding:=margin{Left:=0.0, Top:=0.0, Right:=0.0, Bottom:=0.0}
                                                                                                            Widget:=overlay:
                                                                                                                Slots:=array:

                                                                                                                    #Rarity texture border, when the rarity slot is empty
                                                                                                                    #overlay_slot:
                                                                                                                        #VerticalAlignment:=vertical_alignment.Center
                                                                                                                        #HorizontalAlignment:=horizontal_alignment.Center
                                                                                                                        #Padding:=margin{Left:=0.0, Top:=0.0, Right:=0.0, Bottom:=0.0}
                                                                                                                        #Widget:=texture_block{DefaultImage:=Textures.General.T_GenericBorderSquare2, DefaultDesiredSize:=vector2{X:=30.0, Y:=30.0}, DefaultTint:=MakeColorFromHex("#2e7be388")}
                                                                                                                    
                                                                                                                    #Rarity texture
                                                                                                                    overlay_slot:
                                                                                                                        VerticalAlignment:=vertical_alignment.Center
                                                                                                                        HorizontalAlignment:=horizontal_alignment.Center
                                                                                                                        Padding:=margin{Left:=0.0, Top:=0.0, Right:=0.0, Bottom:=0.0}
                                                                                                                        Widget:=texture_block{DefaultImage:=Textures.General.T_GenericSquare, DefaultDesiredSize:=vector2{X:=20.0, Y:=20.0}, DefaultTint:=MakeColorFromHex("#684f27FF")}

                                                                                                        #Bracket
                                                                                                        overlay_slot:
                                                                                                            VerticalAlignment:=vertical_alignment.Fill
                                                                                                            HorizontalAlignment:=horizontal_alignment.Fill
                                                                                                            Padding:=margin{Left:=0.0, Top:=0.0, Right:=0.0, Bottom:=0.0}
                                                                                                            Widget:=stack_box:
                                                                                                                Orientation:=orientation.Horizontal
                                                                                                                #two instances of bracket texture to represent the range of rarities
                                                                                                                Slots:=array:

                                                                                                                    #stack_box_slot:
                                                                                                                        #VerticalAlignment:=vertical_alignment.Center
                                                                                                                        #HorizontalAlignment:=horizontal_alignment.Left
                                                                                                                        #Distribution:=option{1.0}
                                                                                                                        #Padding:=margin{Left:=0.0, Top:=0.0, Right:=0.0, Bottom:=0.0}
                                                                                                                        #Widget:=texture_block{DefaultImage:=Textures.General.T_RarityBracketLeft, DefaultDesiredSize:=vector2{X:=18.0, Y:=36.0}, DefaultTint:=MakeColorFromHex("#FFFFFF44")}

                                                                                                                    stack_box_slot:
                                                                                                                        VerticalAlignment:=vertical_alignment.Center
                                                                                                                        HorizontalAlignment:=horizontal_alignment.Right
                                                                                                                        Distribution:=option{1.0}
                                                                                                                        Padding:=margin{Left:=0.0, Top:=0.0, Right:=0.0, Bottom:=0.0}
                                                                                                                        Widget:=texture_block{DefaultImage:=Textures.General.T_RarityBracketRight, DefaultDesiredSize:=vector2{X:=18.0, Y:=36.0}, DefaultTint:=MakeColorFromHex("#a3a6aaff")}

                                                                                            stack_box_slot:
                                                                                                VerticalAlignment:=vertical_alignment.Center
                                                                                                HorizontalAlignment:=horizontal_alignment.Center
                                                                                                Distribution:=option{1.0}
                                                                                                Padding:=margin{Left:=0.0, Top:=0.0, Right:=0.0, Bottom:=0.0}
                                                                                                Widget:=overlay:
                                                                                                    Slots:=array:
                                                                                                        #Makeshift size box
                                                                                                        overlay_slot:
                                                                                                            VerticalAlignment:=vertical_alignment.Center
                                                                                                            HorizontalAlignment:=horizontal_alignment.Center
                                                                                                            Padding:=margin{Left:=0.0, Top:=0.0, Right:=0.0, Bottom:=0.0}
                                                                                                            Widget:=texture_block{DefaultImage:=Textures.General.T_GenSquTrans, DefaultDesiredSize:=vector2{X:=30.0, Y:=40.0}, DefaultTint:=MakeColorFromHex("#0000FF08")}

                                                                                                        #Overlay for rarity border and texture
                                                                                                        overlay_slot:
                                                                                                            VerticalAlignment:=vertical_alignment.Center
                                                                                                            HorizontalAlignment:=horizontal_alignment.Center
                                                                                                            Padding:=margin{Left:=0.0, Top:=0.0, Right:=0.0, Bottom:=0.0}
                                                                                                            Widget:=overlay:
                                                                                                                Slots:=array:
                                                                                                        
                                                                                                                    #Rarity texture border, when the rarity slot is empty
                                                                                                                    overlay_slot:
                                                                                                                        VerticalAlignment:=vertical_alignment.Center
                                                                                                                        HorizontalAlignment:=horizontal_alignment.Center
                                                                                                                        Padding:=margin{Left:=0.0, Top:=0.0, Right:=0.0, Bottom:=0.0}
                                                                                                                        Widget:=texture_block{DefaultImage:=Textures.General.T_GenericBorderSquare2, DefaultDesiredSize:=vector2{X:=30.0, Y:=30.0}, DefaultTint:=MakeColorFromHex("#174583FF")}
                                                                                                                    
                                                                                                                    #Rarity texture
                                                                                                                    #overlay_slot:
                                                                                                                        #VerticalAlignment:=vertical_alignment.Center
                                                                                                                        #HorizontalAlignment:=horizontal_alignment.Center
                                                                                                                        #Padding:=margin{Left:=0.0, Top:=0.0, Right:=0.0, Bottom:=0.0}
                                                                                                                        #Widget:=texture_block{DefaultImage:=Textures.General.T_GenericSquare, DefaultDesiredSize:=vector2{X:=30.0, Y:=30.0}, DefaultTint:=MakeColorFromHex("#eacf19FF")}
                                                                                                                    
                                                                                                        #Bracket
                                                                                                        overlay_slot:
                                                                                                            VerticalAlignment:=vertical_alignment.Fill
                                                                                                            HorizontalAlignment:=horizontal_alignment.Fill
                                                                                                            Padding:=margin{Left:=0.0, Top:=0.0, Right:=0.0, Bottom:=0.0}
                                                                                                            Widget:=stack_box:
                                                                                                                Orientation:=orientation.Horizontal
                                                                                                                #two instances of bracket texture to represent the range of rarities
                                                                                                                Slots:=array:

                                                                                                                    #stack_box_slot:
                                                                                                                        #VerticalAlignment:=vertical_alignment.Center
                                                                                                                        #HorizontalAlignment:=horizontal_alignment.Left
                                                                                                                        #Distribution:=option{1.0}
                                                                                                                        #Padding:=margin{Left:=0.0, Top:=0.0, Right:=0.0, Bottom:=0.0}
                                                                                                                        #Widget:=texture_block{DefaultImage:=Textures.General.T_RarityBracketLeft, DefaultDesiredSize:=vector2{X:=18.0, Y:=36.0}, DefaultTint:=MakeColorFromHex("#FFFFFF44")}

                                                                                                                    #stack_box_slot:
                                                                                                                        #VerticalAlignment:=vertical_alignment.Center
                                                                                                                        #HorizontalAlignment:=horizontal_alignment.Right
                                                                                                                        #Distribution:=option{1.0}
                                                                                                                        #Padding:=margin{Left:=0.0, Top:=0.0, Right:=0.0, Bottom:=0.0}
                                                                                                                        #Widget:=texture_block{DefaultImage:=Textures.General.T_RarityBracketRight, DefaultDesiredSize:=vector2{X:=18.0, Y:=36.0}, DefaultTint:=MakeColorFromHex("#FFFFFF44")}

                    #overlay_slot:
                        #VerticalAlignment:=vertical_alignment.Fill
                        #HorizontalAlignment:=horizontal_alignment.Fill
                        #Padding:=margin{Left:=0.0, Top:=0.0, Right:=0.0, Bottom:=0.0}
                        #Widget:=texture_block{DefaultImage:=Textures.General.T_LevelPointer, DefaultDesiredSize:=vector2{X:=20.0, Y:=20.0}, DefaultTint:=MakeColorFromHex("#7de2FFFF")}

            #Definition for the Fifth level row
            LevelRow5:=overlay:
                Slots:=array:
                    overlay_slot:
                        #Makeshift size box
                        VerticalAlignment:=vertical_alignment.Center
                        HorizontalAlignment:=horizontal_alignment.Center
                        Padding:=margin{Left:=0.0, Top:=0.0, Right:=0.0, Bottom:=0.0}
                        Widget:=texture_block{DefaultImage:=Textures.General.T_GenSquTrans, DefaultDesiredSize:=vector2{X:=1250.0, Y:=114.0}, DefaultTint:=MakeColorFromHex("#0000FFFF")}

                    overlay_slot:
                        #Rarity level content
                        VerticalAlignment:=vertical_alignment.Center
                        HorizontalAlignment:=horizontal_alignment.Left
                        Padding:=margin{Left:=0.0, Top:=0.0, Right:=0.0, Bottom:=0.0}
                        Widget:=overlay:
                            Slots:=array:

                                overlay_slot:
                                    #Makeshift size box
                                    VerticalAlignment:=vertical_alignment.Center
                                    HorizontalAlignment:=horizontal_alignment.Center
                                    Padding:=margin{Left:=0.0, Top:=0.0, Right:=0.0, Bottom:=0.0}
                                    Widget:=texture_block{DefaultImage:=Textures.General.T_GenericSquare, DefaultDesiredSize:=vector2{X:=1190.0, Y:=114.0}, DefaultTint:=MakeColorFromHex("#010a1832")}

                                overlay_slot:
                                    #Highlight texture, over the content but under the rarity flag
                                    VerticalAlignment:=vertical_alignment.Center
                                    HorizontalAlignment:=horizontal_alignment.Center
                                    Padding:=margin{Left:=0.0, Top:=0.0, Right:=0.0, Bottom:=0.0}
                                    Widget:=texture_block{DefaultImage:=Textures.General.T_LevelBorder, DefaultDesiredSize:=vector2{X:=1190.0, Y:=114.0}, DefaultTint:=MakeColorFromHex("#174583FF")}

                                    #Horizontal stack box
                                overlay_slot:
                                    VerticalAlignment:=vertical_alignment.Fill
                                    HorizontalAlignment:=horizontal_alignment.Left
                                    Padding:=margin{Left:=0.0, Top:=0.0, Right:=0.0, Bottom:=0.0}
                                    Widget:=stack_box:
                                        Orientation:=orientation.Horizontal
                                        Slots:=array:
                                            #Level flag
                                            stack_box_slot:
                                                VerticalAlignment:=vertical_alignment.Top
                                                HorizontalAlignment:=horizontal_alignment.Left
                                                Padding:=margin{Left:=0.0, Top:=0.0, Right:=15.0, Bottom:=0.0}
                                                Widget:=overlay:

                                                    Slots:=array:
                                                        overlay_slot:
                                                            VerticalAlignment:=vertical_alignment.Center
                                                            HorizontalAlignment:=horizontal_alignment.Center
                                                            Padding:=margin{Left:=0.0, Top:=0.0, Right:=0.0, Bottom:=0.0}
                                                            #Tier level makeshift size box
                                                            Widget:=texture_block{DefaultImage:=Textures.General.T_GenSquTrans, DefaultDesiredSize:=vector2{X:=150.0, Y:=114.0}, DefaultTint:=MakeColorFromHex("#FF0000FF")}

                                                        overlay_slot:
                                                            VerticalAlignment:=vertical_alignment.Top
                                                            HorizontalAlignment:=horizontal_alignment.Left
                                                            Padding:=margin{Left:=0.0, Top:=0.0, Right:=0.0, Bottom:=0.0}
                                                            #Selected level yellow flag
                                                            Widget:=overlay:

                                                                Slots:=array:
                                                                    #overlay_slot:
                                                                        #VerticalAlignment:=vertical_alignment.Center
                                                                        #HorizontalAlignment:=horizontal_alignment.Center
                                                                        #Padding:=margin{Left:=0.0, Top:=0.0, Right:=0.0, Bottom:=0.0}
                                                                        #level flag texture
                                                                        #Widget:=texture_block{DefaultImage:=Textures.General.T_GenLevelFlag, DefaultDesiredSize:=vector2{X:=140.0, Y:=45.0}, DefaultTint:=MakeColorFromHex("#f7ff1aFF")}

                                                                    #overlay_slot:
                                                                        #VerticalAlignment:=vertical_alignment.Center
                                                                        #HorizontalAlignment:=horizontal_alignment.Center
                                                                        #Padding:=margin{Left:=0.0, Top:=0.0, Right:=0.0, Bottom:=0.0}
                                                                        #level flag text
                                                                        #Widget:=text_block{DefaultText := StringToMessage("Level 0"), DefaultTextColor:=MakeColorFromHex("#000000FF")}

                                                        overlay_slot:
                                                            VerticalAlignment:=vertical_alignment.Fill
                                                            HorizontalAlignment:=horizontal_alignment.Center
                                                            Padding:=margin{Left:=0.0, Top:=5.0, Right:=0.0, Bottom:=5.0}
                                                            #Unselected level and lock
                                                            Widget:=overlay:
                                                                Slots:=array:
                                                                    
                                                                    overlay_slot:
                                                                        VerticalAlignment:=vertical_alignment.Top
                                                                        HorizontalAlignment:=horizontal_alignment.Center
                                                                        Padding:=margin{Left:=0.0, Top:=0.0, Right:=0.0, Bottom:=0.0}
                                                                        #level flag texture
                                                                        Widget:=text_block{DefaultText := StringToMessage("Level 5"), DefaultTextColor:=MakeColorFromHex("#a3a6aaff")}

                                                                    overlay_slot:
                                                                        VerticalAlignment:=vertical_alignment.Bottom
                                                                        HorizontalAlignment:=horizontal_alignment.Center
                                                                        Padding:=margin{Left:=0.0, Top:=0.0, Right:=0.0, Bottom:=0.0}
                                                                        #level flag texture
                                                                        Widget:=texture_block{DefaultImage:=Textures.General.T_Lock, DefaultDesiredSize:=vector2{X:=52.0, Y:=58.0}, DefaultTint:=MakeColorFromHex("#a3a6aaff")}
                                            
                                            #Level weapon roster
                                            stack_box_slot:
                                                VerticalAlignment:=vertical_alignment.Center
                                                HorizontalAlignment:=horizontal_alignment.Center
                                                Padding:=margin{Left:=0.0, Top:=0.0, Right:=8.0, Bottom:=0.0}
                                                Widget:=overlay:
                                                    Slots:=array:
                                                        #Weapon Icon - Pistols
                                                        overlay_slot:
                                                            VerticalAlignment:=vertical_alignment.Center
                                                            HorizontalAlignment:=horizontal_alignment.Center
                                                            Padding:=margin{Left:=0.0, Top:=0.0, Right:=0.0, Bottom:=0.0}
                                                            #Weapon icon BG
                                                            Widget:=texture_block{DefaultImage:=Textures.General.T_GenericSquare, DefaultDesiredSize:=vector2{X:=78.0, Y:=78.0}, DefaultTint:=MakeColorFromHex("#000c1d88")}

                                                        overlay_slot:
                                                            VerticalAlignment:=vertical_alignment.Center
                                                            HorizontalAlignment:=horizontal_alignment.Center
                                                            Padding:=margin{Left:=0.0, Top:=0.0, Right:=0.0, Bottom:=0.0}
                                                            #Weapon icon Border
                                                            Widget:=texture_block{DefaultImage:=Textures.General.T_GenericBorderSquare, DefaultDesiredSize:=vector2{X:=78.0, Y:=78.0}, DefaultTint:=MakeColorFromHex("#174583FF")}

                                                        overlay_slot:
                                                            VerticalAlignment:=vertical_alignment.Top
                                                            HorizontalAlignment:=horizontal_alignment.Center
                                                            Padding:=margin{Left:=0.0, Top:=5.0, Right:=0.0, Bottom:=0.0}
                                                            #Weapon icon - pistol
                                                            Widget:=texture_block{DefaultImage:=Textures.Items.T_IconPistol, DefaultDesiredSize:=vector2{X:=50.0, Y:=50.0}, DefaultTint:=MakeColorFromHex("#a3a6aaff")}

                                                        overlay_slot:
                                                            VerticalAlignment:=vertical_alignment.Bottom
                                                            HorizontalAlignment:=horizontal_alignment.Center
                                                            Padding:=margin{Left:=0.0, Top:=0.0, Right:=0.0, Bottom:=5.0}
                                                            #Weapon header - pistol
                                                            Widget:=texture_block{DefaultImage:=Textures.Items.T_HeaderPistol, DefaultDesiredSize:=vector2{X:=78.0, Y:=18.0}, DefaultTint:=MakeColorFromHex("#a3a6aaff")}
                                                        
                                            stack_box_slot:
                                                VerticalAlignment:=vertical_alignment.Center
                                                HorizontalAlignment:=horizontal_alignment.Center
                                                Padding:=margin{Left:=0.0, Top:=0.0, Right:=8.0, Bottom:=0.0}
                                                Widget:=overlay:
                                                    Slots:=array:
                                                        #Weapon Icon - Shotgun
                                                        overlay_slot:
                                                            VerticalAlignment:=vertical_alignment.Center
                                                            HorizontalAlignment:=horizontal_alignment.Center
                                                            Padding:=margin{Left:=0.0, Top:=0.0, Right:=0.0, Bottom:=0.0}
                                                            #Weapon icon BG
                                                            Widget:=texture_block{DefaultImage:=Textures.General.T_GenericSquare, DefaultDesiredSize:=vector2{X:=78.0, Y:=78.0}, DefaultTint:=MakeColorFromHex("#000c1d88")}

                                                        overlay_slot:
                                                            VerticalAlignment:=vertical_alignment.Center
                                                            HorizontalAlignment:=horizontal_alignment.Center
                                                            Padding:=margin{Left:=0.0, Top:=0.0, Right:=0.0, Bottom:=0.0}
                                                            #Weapon icon Border
                                                            Widget:=texture_block{DefaultImage:=Textures.General.T_GenericBorderSquare, DefaultDesiredSize:=vector2{X:=78.0, Y:=78.0}, DefaultTint:=MakeColorFromHex("#174583FF")}

                                                        overlay_slot:
                                                            VerticalAlignment:=vertical_alignment.Top
                                                            HorizontalAlignment:=horizontal_alignment.Center
                                                            Padding:=margin{Left:=0.0, Top:=5.0, Right:=0.0, Bottom:=0.0}
                                                            #Weapon icon - shotgun
                                                            Widget:=texture_block{DefaultImage:=Textures.Items.T_IconShotgun, DefaultDesiredSize:=vector2{X:=50.0, Y:=50.0}, DefaultTint:=MakeColorFromHex("#a3a6aaff")}

                                                        overlay_slot:
                                                            VerticalAlignment:=vertical_alignment.Bottom
                                                            HorizontalAlignment:=horizontal_alignment.Center
                                                            Padding:=margin{Left:=0.0, Top:=0.0, Right:=0.0, Bottom:=5.0}
                                                            #Weapon header - shotgun
                                                            Widget:=texture_block{DefaultImage:=Textures.Items.T_HeaderShotgun, DefaultDesiredSize:=vector2{X:=78.0, Y:=18.0}, DefaultTint:=MakeColorFromHex("#a3a6aaff")}

                                            stack_box_slot:
                                                VerticalAlignment:=vertical_alignment.Center
                                                HorizontalAlignment:=horizontal_alignment.Center
                                                Padding:=margin{Left:=0.0, Top:=0.0, Right:=8.0, Bottom:=0.0}
                                                Widget:=overlay:
                                                    Slots:=array:
                                                        #Weapon Icon - SMG
                                                        overlay_slot:
                                                            VerticalAlignment:=vertical_alignment.Top
                                                            HorizontalAlignment:=horizontal_alignment.Center
                                                            Padding:=margin{Left:=0.0, Top:=0.0, Right:=0.0, Bottom:=0.0}
                                                            #Weapon icon BG
                                                            Widget:=texture_block{DefaultImage:=Textures.General.T_GenericSquare, DefaultDesiredSize:=vector2{X:=78.0, Y:=78.0}, DefaultTint:=MakeColorFromHex("#000c1d88")}

                                                        overlay_slot:
                                                            VerticalAlignment:=vertical_alignment.Center
                                                            HorizontalAlignment:=horizontal_alignment.Center
                                                            Padding:=margin{Left:=0.0, Top:=0.0, Right:=0.0, Bottom:=0.0}
                                                            #Weapon icon Border
                                                            Widget:=texture_block{DefaultImage:=Textures.General.T_GenericBorderSquare, DefaultDesiredSize:=vector2{X:=78.0, Y:=78.0}, DefaultTint:=MakeColorFromHex("#174583FF")}

                                                        overlay_slot:
                                                            VerticalAlignment:=vertical_alignment.Top
                                                            HorizontalAlignment:=horizontal_alignment.Center
                                                            Padding:=margin{Left:=0.0, Top:=5.0, Right:=0.0, Bottom:=0.0}
                                                            #Weapon icon - SMG
                                                            Widget:=texture_block{DefaultImage:=Textures.Items.T_IconSMG, DefaultDesiredSize:=vector2{X:=50.0, Y:=50.0}, DefaultTint:=MakeColorFromHex("#a3a6aaff")}

                                                        overlay_slot:
                                                            VerticalAlignment:=vertical_alignment.Bottom
                                                            HorizontalAlignment:=horizontal_alignment.Center
                                                            Padding:=margin{Left:=0.0, Top:=0.0, Right:=0.0, Bottom:=5.0}
                                                            #Weapon header - SMG
                                                            Widget:=texture_block{DefaultImage:=Textures.Items.T_HeaderSMGs, DefaultDesiredSize:=vector2{X:=78.0, Y:=18.0}, DefaultTint:=MakeColorFromHex("#a3a6aaff")}
                                                            

                                            stack_box_slot:
                                                VerticalAlignment:=vertical_alignment.Center
                                                HorizontalAlignment:=horizontal_alignment.Center
                                                Padding:=margin{Left:=0.0, Top:=0.0, Right:=8.0, Bottom:=0.0}
                                                Widget:=overlay:
                                                    Slots:=array:
                                                        #Weapon Icon - Rifle
                                                        overlay_slot:
                                                            VerticalAlignment:=vertical_alignment.Center
                                                            HorizontalAlignment:=horizontal_alignment.Center
                                                            Padding:=margin{Left:=0.0, Top:=0.0, Right:=0.0, Bottom:=0.0}
                                                            #Weapon icon BG
                                                            Widget:=texture_block{DefaultImage:=Textures.General.T_GenericSquare, DefaultDesiredSize:=vector2{X:=78.0, Y:=78.0}, DefaultTint:=MakeColorFromHex("#000c1d88")}

                                                        overlay_slot:
                                                            VerticalAlignment:=vertical_alignment.Top
                                                            HorizontalAlignment:=horizontal_alignment.Center
                                                            Padding:=margin{Left:=0.0, Top:=0.0, Right:=0.0, Bottom:=0.0}
                                                            #Weapon icon Border
                                                            Widget:=texture_block{DefaultImage:=Textures.General.T_GenericBorderSquare, DefaultDesiredSize:=vector2{X:=78.0, Y:=78.0}, DefaultTint:=MakeColorFromHex("#174583FF")}

                                                        overlay_slot:
                                                            VerticalAlignment:=vertical_alignment.Top
                                                            HorizontalAlignment:=horizontal_alignment.Center
                                                            Padding:=margin{Left:=0.0, Top:=5.0, Right:=0.0, Bottom:=0.0}
                                                            #Weapon icon - Rifle
                                                            Widget:=texture_block{DefaultImage:=Textures.Items.T_IconRifle, DefaultDesiredSize:=vector2{X:=50.0, Y:=50.0}, DefaultTint:=MakeColorFromHex("#a3a6aaff")}

                                                        overlay_slot:
                                                            VerticalAlignment:=vertical_alignment.Bottom
                                                            HorizontalAlignment:=horizontal_alignment.Center
                                                            Padding:=margin{Left:=0.0, Top:=0.0, Right:=0.0, Bottom:=5.0}
                                                            #Weapon header - Rifle
                                                            Widget:=texture_block{DefaultImage:=Textures.Items.T_HeaderRifle, DefaultDesiredSize:=vector2{X:=78.0, Y:=18.0}, DefaultTint:=MakeColorFromHex("#a3a6aaff")}

                                            stack_box_slot:
                                                VerticalAlignment:=vertical_alignment.Center
                                                HorizontalAlignment:=horizontal_alignment.Center
                                                Padding:=margin{Left:=0.0, Top:=0.0, Right:=8.0, Bottom:=0.0}
                                                Widget:=overlay:
                                                    Slots:=array:
                                                        #Weapon Icon - Explosive
                                                        overlay_slot:
                                                            VerticalAlignment:=vertical_alignment.Center
                                                            HorizontalAlignment:=horizontal_alignment.Center
                                                            Padding:=margin{Left:=0.0, Top:=0.0, Right:=0.0, Bottom:=0.0}
                                                            #Weapon icon BG
                                                            Widget:=texture_block{DefaultImage:=Textures.General.T_GenericSquare, DefaultDesiredSize:=vector2{X:=78.0, Y:=78.0}, DefaultTint:=MakeColorFromHex("#000c1d88")}

                                                        overlay_slot:
                                                            VerticalAlignment:=vertical_alignment.Center
                                                            HorizontalAlignment:=horizontal_alignment.Center
                                                            Padding:=margin{Left:=0.0, Top:=0.0, Right:=0.0, Bottom:=0.0}
                                                            #Weapon icon Border
                                                            Widget:=texture_block{DefaultImage:=Textures.General.T_GenericBorderSquare, DefaultDesiredSize:=vector2{X:=78.0, Y:=78.0}, DefaultTint:=MakeColorFromHex("#174583FF")}

                                                        overlay_slot:
                                                            VerticalAlignment:=vertical_alignment.Top
                                                            HorizontalAlignment:=horizontal_alignment.Center
                                                            Padding:=margin{Left:=0.0, Top:=5.0, Right:=0.0, Bottom:=0.0}
                                                            #Weapon icon - Explosive
                                                            Widget:=texture_block{DefaultImage:=Textures.Items.T_IconExplosive, DefaultDesiredSize:=vector2{X:=50.0, Y:=50.0}, DefaultTint:=MakeColorFromHex("#a3a6aaff")}

                                                        overlay_slot:
                                                            VerticalAlignment:=vertical_alignment.Bottom
                                                            HorizontalAlignment:=horizontal_alignment.Center
                                                            Padding:=margin{Left:=0.0, Top:=0.0, Right:=0.0, Bottom:=5.0}
                                                            #Weapon header - Explosive
                                                            Widget:=texture_block{DefaultImage:=Textures.Items.T_HeaderExplosive, DefaultDesiredSize:=vector2{X:=78.0, Y:=18.0}, DefaultTint:=MakeColorFromHex("#a3a6aaff")}

                                            stack_box_slot:
                                                VerticalAlignment:=vertical_alignment.Center
                                                HorizontalAlignment:=horizontal_alignment.Center
                                                Padding:=margin{Left:=0.0, Top:=0.0, Right:=8.0, Bottom:=0.0}
                                                Widget:=overlay:
                                                    Slots:=array:
                                                        #Weapon Icon - Special
                                                        overlay_slot:
                                                            VerticalAlignment:=vertical_alignment.Center
                                                            HorizontalAlignment:=horizontal_alignment.Center
                                                            Padding:=margin{Left:=0.0, Top:=0.0, Right:=0.0, Bottom:=0.0}
                                                            #Weapon icon BG
                                                            Widget:=texture_block{DefaultImage:=Textures.General.T_GenericSquare, DefaultDesiredSize:=vector2{X:=78.0, Y:=78.0}, DefaultTint:=MakeColorFromHex("#000c1d88")}

                                                        overlay_slot:
                                                            VerticalAlignment:=vertical_alignment.Center
                                                            HorizontalAlignment:=horizontal_alignment.Center
                                                            Padding:=margin{Left:=0.0, Top:=0.0, Right:=0.0, Bottom:=0.0}
                                                            #Weapon icon Border
                                                            Widget:=texture_block{DefaultImage:=Textures.General.T_GenericBorderSquare, DefaultDesiredSize:=vector2{X:=78.0, Y:=78.0}, DefaultTint:=MakeColorFromHex("#174583FF")}

                                                        overlay_slot:
                                                            VerticalAlignment:=vertical_alignment.Top
                                                            HorizontalAlignment:=horizontal_alignment.Center
                                                            Padding:=margin{Left:=0.0, Top:=5.0, Right:=0.0, Bottom:=0.0}
                                                            #Weapon icon - special
                                                            Widget:=texture_block{DefaultImage:=Textures.Items.T_IconArtifact, DefaultDesiredSize:=vector2{X:=50.0, Y:=50.0}, DefaultTint:=MakeColorFromHex("#a3a6aaff")}

                                                        overlay_slot:
                                                            VerticalAlignment:=vertical_alignment.Bottom
                                                            HorizontalAlignment:=horizontal_alignment.Center
                                                            Padding:=margin{Left:=0.0, Top:=0.0, Right:=0.0, Bottom:=5.0}
                                                            #Weapon header - Special
                                                            Widget:=texture_block{DefaultImage:=Textures.Items.T_HeaderSpecial, DefaultDesiredSize:=vector2{X:=78.0, Y:=18.0}, DefaultTint:=MakeColorFromHex("#a3a6aaff")}

                                            stack_box_slot:
                                                VerticalAlignment:=vertical_alignment.Center
                                                HorizontalAlignment:=horizontal_alignment.Center
                                                Padding:=margin{Left:=0.0, Top:=0.0, Right:=8.0, Bottom:=0.0}
                                                Widget:=overlay:
                                                    Slots:=array:
                                                        #Weapon Icon - DMR Sniper
                                                        overlay_slot:
                                                            VerticalAlignment:=vertical_alignment.Center
                                                            HorizontalAlignment:=horizontal_alignment.Center
                                                            Padding:=margin{Left:=0.0, Top:=0.0, Right:=0.0, Bottom:=0.0}
                                                            #Weapon icon BG
                                                            Widget:=texture_block{DefaultImage:=Textures.General.T_GenericSquare, DefaultDesiredSize:=vector2{X:=78.0, Y:=78.0}, DefaultTint:=MakeColorFromHex("#000c1d88")}

                                                        overlay_slot:
                                                            VerticalAlignment:=vertical_alignment.Center
                                                            HorizontalAlignment:=horizontal_alignment.Center
                                                            Padding:=margin{Left:=0.0, Top:=0.0, Right:=0.0, Bottom:=0.0}
                                                            #Weapon icon Border
                                                            Widget:=texture_block{DefaultImage:=Textures.General.T_GenericBorderSquare, DefaultDesiredSize:=vector2{X:=78.0, Y:=78.0}, DefaultTint:=MakeColorFromHex("#174583FF")}

                                                        overlay_slot:
                                                            VerticalAlignment:=vertical_alignment.Top
                                                            HorizontalAlignment:=horizontal_alignment.Center
                                                            Padding:=margin{Left:=0.0, Top:=5.0, Right:=0.0, Bottom:=0.0}
                                                            #Weapon icon - Sniper
                                                            Widget:=texture_block{DefaultImage:=Textures.Items.T_IconSniper, DefaultDesiredSize:=vector2{X:=50.0, Y:=50.0}, DefaultTint:=MakeColorFromHex("#a3a6aaff")}

                                                        overlay_slot:
                                                            VerticalAlignment:=vertical_alignment.Bottom
                                                            HorizontalAlignment:=horizontal_alignment.Center
                                                            Padding:=margin{Left:=0.0, Top:=0.0, Right:=0.0, Bottom:=5.0}
                                                            #Weapon header - DMR Sniper
                                                            Widget:=texture_block{DefaultImage:=Textures.Items.T_HeaderDMRSniper, DefaultDesiredSize:=vector2{X:=78.0, Y:=18.0}, DefaultTint:=MakeColorFromHex("#a3a6aaff")}

                                            stack_box_slot:
                                                VerticalAlignment:=vertical_alignment.Fill
                                                HorizontalAlignment:=horizontal_alignment.Center #Center
                                                Padding:=margin{Left:=0.0, Top:=0.0, Right:=0.0, Bottom:=0.0}
                                                Widget:=overlay:
                                                    Slots:=array:
                                                        overlay_slot:
                                                            #Makeshift size box
                                                            VerticalAlignment:=vertical_alignment.Center
                                                            HorizontalAlignment:=horizontal_alignment.Center
                                                            Padding:=margin{Left:=0.0, Top:=0.0, Right:=0.0, Bottom:=0.0}
                                                            Widget:=texture_block{DefaultImage:=Textures.General.T_GenSquTrans, DefaultDesiredSize:=vector2{X:=418.0, Y:=100.0}, DefaultTint:=MakeColorFromHex("#00FF0011")}

                                                        overlay_slot:
                                                            #Vertical stack box to add all the rarity information
                                                            VerticalAlignment:=vertical_alignment.Center
                                                            HorizontalAlignment:=horizontal_alignment.Center
                                                            Padding:=margin{Left:=0.0, Top:=0.0, Right:=0.0, Bottom:=0.0}
                                                            Widget:=stack_box:
                                                                Orientation:=orientation.Vertical
                                                                Slots:=array:
                                                                    stack_box_slot:
                                                                        #Horizontal Stack box for max rarity available
                                                                        VerticalAlignment:=vertical_alignment.Top
                                                                        HorizontalAlignment:=horizontal_alignment.Center
                                                                        Padding:=margin{Left:=0.0, Top:=0.0, Right:=0.0, Bottom:=0.0}
                                                                        Widget:=stack_box:
                                                                            Orientation:=orientation.Horizontal
                                                                            Slots:=array:
                                                                                stack_box_slot:
                                                                                    VerticalAlignment:=vertical_alignment.Center
                                                                                    HorizontalAlignment:=horizontal_alignment.Center
                                                                                    Padding:=margin{Left:=0.0, Top:=0.0, Right:=0.0, Bottom:=0.0}
                                                                                    Widget:=text_block{DefaultText := StringToMessage("Rarity chance up to "), DefaultTextColor:=MakeColorFromHex("#a3a6aaff")}

                                                                                stack_box_slot:
                                                                                    VerticalAlignment:=vertical_alignment.Center
                                                                                    HorizontalAlignment:=horizontal_alignment.Center
                                                                                    Padding:=margin{Left:=0.0, Top:=0.0, Right:=0.0, Bottom:=0.0}
                                                                                    Widget:=texture_block{DefaultImage:=Textures.Rarity.T_FlagMythic, DefaultDesiredSize:=vector2{X:=104.0, Y:=36.0}, DefaultTint:=MakeColorFromHex("#a3a6aaff")}

                                                                    stack_box_slot:
                                                                        #Horizontal overlay for overall rarity scheme
                                                                        VerticalAlignment:=vertical_alignment.Bottom
                                                                        HorizontalAlignment:=horizontal_alignment.Center
                                                                        Padding:=margin{Left:=0.0, Top:=0.0, Right:=0.0, Bottom:=0.0}
                                                                        Widget:=overlay:
                                                                            Slots:=array:
                                                                                
                                                                                #Makeshift size box for the rarity scheme
                                                                                overlay_slot:
                                                                                    VerticalAlignment:=vertical_alignment.Center
                                                                                    HorizontalAlignment:=horizontal_alignment.Center
                                                                                    Padding:=margin{Left:=0.0, Top:=0.0, Right:=0.0, Bottom:=0.0}
                                                                                    Widget:=texture_block{DefaultImage:=Textures.General.T_GenSquTrans, DefaultDesiredSize:=vector2{X:=255.0, Y:=50.0}, DefaultTint:=MakeColorFromHex("#FF000022")}#Red frame

                                                                                #Visible border for rarity layout
                                                                                overlay_slot:
                                                                                    VerticalAlignment:=vertical_alignment.Center
                                                                                    HorizontalAlignment:=horizontal_alignment.Center
                                                                                    Padding:=margin{Left:=0.0, Top:=0.0, Right:=0.0, Bottom:=0.0}
                                                                                    Widget:=texture_block{DefaultImage:=Textures.General.T_GenericBorderSquare3, DefaultDesiredSize:=vector2{X:=255.0, Y:=40.0}, DefaultTint:=MakeColorFromHex("#174583FF")}

                                                                                #Horizontal Stack box for rarity slots
                                                                                overlay_slot:
                                                                                    VerticalAlignment:=vertical_alignment.Fill
                                                                                    HorizontalAlignment:=horizontal_alignment.Fill
                                                                                    Padding:=margin{Left:=5.0, Top:=0.0, Right:=5.0, Bottom:=0.0}
                                                                                    Widget:=stack_box:
                                                                                        Orientation:=orientation.Horizontal
                                                                                        Slots:=array:
                                                                                            #Overlay for each rarity slots, including the pointing arrow
                                                                                            stack_box_slot:
                                                                                                VerticalAlignment:=vertical_alignment.Center
                                                                                                HorizontalAlignment:=horizontal_alignment.Center
                                                                                                Distribution:=option{1.0}
                                                                                                Padding:=margin{Left:=0.0, Top:=0.0, Right:=0.0, Bottom:=0.0}
                                                                                                Widget:=overlay:
                                                                                                    Slots:=array:
                                                                                                        #Makeshift size box
                                                                                                        overlay_slot:
                                                                                                            VerticalAlignment:=vertical_alignment.Center
                                                                                                            HorizontalAlignment:=horizontal_alignment.Center
                                                                                                            Padding:=margin{Left:=0.0, Top:=0.0, Right:=0.0, Bottom:=0.0}
                                                                                                            Widget:=texture_block{DefaultImage:=Textures.General.T_GenSquTrans, DefaultDesiredSize:=vector2{X:=30.0, Y:=40.0}, DefaultTint:=MakeColorFromHex("#0000FF08")}

                                                                                                        #Overlay for rarity border and texture
                                                                                                        overlay_slot:
                                                                                                            VerticalAlignment:=vertical_alignment.Center
                                                                                                            HorizontalAlignment:=horizontal_alignment.Center
                                                                                                            Padding:=margin{Left:=0.0, Top:=0.0, Right:=0.0, Bottom:=0.0}
                                                                                                            Widget:=overlay:
                                                                                                                Slots:=array:
                                                                                                                    
                                                                                                                    #Rarity texture border, when the rarity slot is empty
                                                                                                                    overlay_slot:
                                                                                                                        VerticalAlignment:=vertical_alignment.Center
                                                                                                                        HorizontalAlignment:=horizontal_alignment.Center
                                                                                                                        Padding:=margin{Left:=0.0, Top:=0.0, Right:=0.0, Bottom:=0.0}
                                                                                                                        Widget:=texture_block{DefaultImage:=Textures.General.T_GenericBorderSquare2, DefaultDesiredSize:=vector2{X:=30.0, Y:=30.0}, DefaultTint:=MakeColorFromHex("#174583FF")}

                                                                                                                    #Rarity texture
                                                                                                                    #overlay_slot:
                                                                                                                        #VerticalAlignment:=vertical_alignment.Center
                                                                                                                        #HorizontalAlignment:=horizontal_alignment.Center
                                                                                                                        #Padding:=margin{Left:=0.0, Top:=0.0, Right:=0.0, Bottom:=0.0}
                                                                                                                        #Widget:=texture_block{DefaultImage:=Textures.General.T_GenericSquare, DefaultDesiredSize:=vector2{X:=30.0, Y:=30.0}, DefaultTint:=MakeColorFromHex("#979797FF")}
                                                                                                        
                                                                                                                    

                                                                                                        #Bracket
                                                                                                        overlay_slot:
                                                                                                            VerticalAlignment:=vertical_alignment.Fill
                                                                                                            HorizontalAlignment:=horizontal_alignment.Fill
                                                                                                            Padding:=margin{Left:=0.0, Top:=0.0, Right:=0.0, Bottom:=0.0}
                                                                                                            Widget:=stack_box:
                                                                                                                Orientation:=orientation.Horizontal
                                                                                                                #two instances of bracket texture to represent the range of rarities
                                                                                                                Slots:=array:

                                                                                                                    #stack_box_slot:
                                                                                                                        #VerticalAlignment:=vertical_alignment.Center
                                                                                                                        #HorizontalAlignment:=horizontal_alignment.Left
                                                                                                                        #Distribution:=option{1.0}
                                                                                                                        #Padding:=margin{Left:=0.0, Top:=0.0, Right:=0.0, Bottom:=0.0}
                                                                                                                        #Widget:=texture_block{DefaultImage:=Textures.General.T_RarityBracketLeft, DefaultDesiredSize:=vector2{X:=18.0, Y:=36.0}, DefaultTint:=MakeColorFromHex("#FFFFFF44")}

                                                                                                                    #stack_box_slot:
                                                                                                                        #VerticalAlignment:=vertical_alignment.Center
                                                                                                                        #HorizontalAlignment:=horizontal_alignment.Right
                                                                                                                        #Distribution:=option{1.0}
                                                                                                                        #Padding:=margin{Left:=0.0, Top:=0.0, Right:=0.0, Bottom:=0.0}
                                                                                                                        #Widget:=texture_block{DefaultImage:=Textures.General.T_RarityBracketRight, DefaultDesiredSize:=vector2{X:=18.0, Y:=36.0}, DefaultTint:=MakeColorFromHex("#FFFFFF44")}

                                                                                            stack_box_slot:
                                                                                                VerticalAlignment:=vertical_alignment.Center
                                                                                                HorizontalAlignment:=horizontal_alignment.Center
                                                                                                Distribution:=option{1.0}
                                                                                                Padding:=margin{Left:=0.0, Top:=0.0, Right:=0.0, Bottom:=0.0}
                                                                                                Widget:=overlay:
                                                                                                    Slots:=array:
                                                                                                        #Makeshift size box
                                                                                                        overlay_slot:
                                                                                                            VerticalAlignment:=vertical_alignment.Center
                                                                                                            HorizontalAlignment:=horizontal_alignment.Center
                                                                                                            Padding:=margin{Left:=0.0, Top:=0.0, Right:=0.0, Bottom:=0.0}
                                                                                                            Widget:=texture_block{DefaultImage:=Textures.General.T_GenSquTrans, DefaultDesiredSize:=vector2{X:=30.0, Y:=40.0}, DefaultTint:=MakeColorFromHex("#0000FF08")}

                                                                                                        #Overlay for rarity border and texture
                                                                                                        overlay_slot:
                                                                                                            VerticalAlignment:=vertical_alignment.Center
                                                                                                            HorizontalAlignment:=horizontal_alignment.Center
                                                                                                            Padding:=margin{Left:=0.0, Top:=0.0, Right:=0.0, Bottom:=0.0}
                                                                                                            Widget:=overlay:
                                                                                                                Slots:=array:
                                                                                                                    
                                                                                                                    #Rarity texture border, when the rarity slot is empty
                                                                                                                    overlay_slot:
                                                                                                                        VerticalAlignment:=vertical_alignment.Center
                                                                                                                        HorizontalAlignment:=horizontal_alignment.Center
                                                                                                                        Padding:=margin{Left:=0.0, Top:=0.0, Right:=0.0, Bottom:=0.0}
                                                                                                                        Widget:=texture_block{DefaultImage:=Textures.General.T_GenericBorderSquare2, DefaultDesiredSize:=vector2{X:=30.0, Y:=30.0}, DefaultTint:=MakeColorFromHex("#174583FF")}

                                                                                                                    #Rarity texture
                                                                                                                    #overlay_slot:
                                                                                                                        #VerticalAlignment:=vertical_alignment.Center
                                                                                                                        #HorizontalAlignment:=horizontal_alignment.Center
                                                                                                                        #Padding:=margin{Left:=0.0, Top:=0.0, Right:=0.0, Bottom:=0.0}
                                                                                                                        #Widget:=texture_block{DefaultImage:=Textures.General.T_GenericSquare, DefaultDesiredSize:=vector2{X:=30.0, Y:=30.0}, DefaultTint:=MakeColorFromHex("#4ccc2aFF")}

                                                                                                        #Bracket
                                                                                                        overlay_slot:
                                                                                                            VerticalAlignment:=vertical_alignment.Fill
                                                                                                            HorizontalAlignment:=horizontal_alignment.Fill
                                                                                                            Padding:=margin{Left:=0.0, Top:=0.0, Right:=0.0, Bottom:=0.0}
                                                                                                            Widget:=stack_box:
                                                                                                                Orientation:=orientation.Horizontal
                                                                                                                #two instances of bracket texture to represent the range of rarities
                                                                                                                Slots:=array:

                                                                                                                    #stack_box_slot:
                                                                                                                        #VerticalAlignment:=vertical_alignment.Center
                                                                                                                        #HorizontalAlignment:=horizontal_alignment.Left
                                                                                                                        #Distribution:=option{1.0}
                                                                                                                        #Padding:=margin{Left:=0.0, Top:=0.0, Right:=0.0, Bottom:=0.0}
                                                                                                                        #Widget:=texture_block{DefaultImage:=Textures.General.T_RarityBracketLeft, DefaultDesiredSize:=vector2{X:=18.0, Y:=36.0}, DefaultTint:=MakeColorFromHex("#FFFFFF44")}

                                                                                                                    #stack_box_slot:
                                                                                                                        #VerticalAlignment:=vertical_alignment.Center
                                                                                                                        #HorizontalAlignment:=horizontal_alignment.Right
                                                                                                                        #Distribution:=option{1.0}
                                                                                                                        #Padding:=margin{Left:=0.0, Top:=0.0, Right:=0.0, Bottom:=0.0}
                                                                                                                        #Widget:=texture_block{DefaultImage:=Textures.General.T_RarityBracketRight, DefaultDesiredSize:=vector2{X:=18.0, Y:=36.0}, DefaultTint:=MakeColorFromHex("#FFFFFF44")}

                                                                                            stack_box_slot:
                                                                                                VerticalAlignment:=vertical_alignment.Center
                                                                                                HorizontalAlignment:=horizontal_alignment.Center
                                                                                                Distribution:=option{1.0}
                                                                                                Padding:=margin{Left:=0.0, Top:=0.0, Right:=0.0, Bottom:=0.0}
                                                                                                Widget:=overlay:
                                                                                                    Slots:=array:
                                                                                                        #Makeshift size box
                                                                                                        overlay_slot:
                                                                                                            VerticalAlignment:=vertical_alignment.Center
                                                                                                            HorizontalAlignment:=horizontal_alignment.Center
                                                                                                            Padding:=margin{Left:=0.0, Top:=0.0, Right:=0.0, Bottom:=0.0}
                                                                                                            Widget:=texture_block{DefaultImage:=Textures.General.T_GenSquTrans, DefaultDesiredSize:=vector2{X:=30.0, Y:=40.0}, DefaultTint:=MakeColorFromHex("#0000FF08")}

                                                                                                        #Overlay for rarity border and texture
                                                                                                        overlay_slot:
                                                                                                            VerticalAlignment:=vertical_alignment.Center
                                                                                                            HorizontalAlignment:=horizontal_alignment.Center
                                                                                                            Padding:=margin{Left:=0.0, Top:=0.0, Right:=0.0, Bottom:=0.0}
                                                                                                            Widget:=overlay:
                                                                                                                Slots:=array:
                                                                                                                                                                                                                
                                                                                                                    #Rarity texture border, when the rarity slot is empty
                                                                                                                    #overlay_slot:
                                                                                                                        #VerticalAlignment:=vertical_alignment.Center
                                                                                                                        #HorizontalAlignment:=horizontal_alignment.Center
                                                                                                                        #Padding:=margin{Left:=0.0, Top:=0.0, Right:=0.0, Bottom:=0.0}
                                                                                                                        #Widget:=texture_block{DefaultImage:=Textures.General.T_GenericBorderSquare2, DefaultDesiredSize:=vector2{X:=30.0, Y:=30.0}, DefaultTint:=MakeColorFromHex("#2e7be388")}
                                                                                                                                                                                                                            
                                                                                                                    #Rarity texture
                                                                                                                    overlay_slot:
                                                                                                                        VerticalAlignment:=vertical_alignment.Center
                                                                                                                        HorizontalAlignment:=horizontal_alignment.Center
                                                                                                                        Padding:=margin{Left:=0.0, Top:=0.0, Right:=0.0, Bottom:=0.0}
                                                                                                                        Widget:=texture_block{DefaultImage:=Textures.General.T_GenericSquare, DefaultDesiredSize:=vector2{X:=20.0, Y:=20.0}, DefaultTint:=MakeColorFromHex("#183379FF")}
                                                                                                                    
                                                                                                        #Bracket
                                                                                                        overlay_slot:
                                                                                                            VerticalAlignment:=vertical_alignment.Fill
                                                                                                            HorizontalAlignment:=horizontal_alignment.Fill
                                                                                                            Padding:=margin{Left:=0.0, Top:=0.0, Right:=0.0, Bottom:=0.0}
                                                                                                            Widget:=stack_box:
                                                                                                                Orientation:=orientation.Horizontal
                                                                                                                #two instances of bracket texture to represent the range of rarities
                                                                                                                Slots:=array:

                                                                                                                    stack_box_slot:
                                                                                                                        VerticalAlignment:=vertical_alignment.Center
                                                                                                                        HorizontalAlignment:=horizontal_alignment.Left
                                                                                                                        Distribution:=option{1.0}
                                                                                                                        Padding:=margin{Left:=0.0, Top:=0.0, Right:=0.0, Bottom:=0.0}
                                                                                                                        Widget:=texture_block{DefaultImage:=Textures.General.T_RarityBracketLeft, DefaultDesiredSize:=vector2{X:=18.0, Y:=36.0}, DefaultTint:=MakeColorFromHex("#a3a6aaff")}

                                                                                                                    #stack_box_slot:
                                                                                                                        #VerticalAlignment:=vertical_alignment.Center
                                                                                                                        #HorizontalAlignment:=horizontal_alignment.Right
                                                                                                                        #Distribution:=option{1.0}
                                                                                                                        #Padding:=margin{Left:=0.0, Top:=0.0, Right:=0.0, Bottom:=0.0}
                                                                                                                        #Widget:=texture_block{DefaultImage:=Textures.General.T_RarityBracketRight, DefaultDesiredSize:=vector2{X:=18.0, Y:=36.0}, DefaultTint:=MakeColorFromHex("#FFFFFF44")}

                                                                                            stack_box_slot:
                                                                                                VerticalAlignment:=vertical_alignment.Center
                                                                                                HorizontalAlignment:=horizontal_alignment.Center
                                                                                                Distribution:=option{1.0}
                                                                                                Padding:=margin{Left:=0.0, Top:=0.0, Right:=0.0, Bottom:=0.0}
                                                                                                Widget:=overlay:
                                                                                                    Slots:=array:
                                                                                                        #Makeshift size box
                                                                                                        overlay_slot:
                                                                                                            VerticalAlignment:=vertical_alignment.Center
                                                                                                            HorizontalAlignment:=horizontal_alignment.Center
                                                                                                            Padding:=margin{Left:=0.0, Top:=0.0, Right:=0.0, Bottom:=0.0}
                                                                                                            Widget:=texture_block{DefaultImage:=Textures.General.T_GenSquTrans, DefaultDesiredSize:=vector2{X:=30.0, Y:=40.0}, DefaultTint:=MakeColorFromHex("#0000FF08")}

                                                                                                        #Overlay for rarity border and texture
                                                                                                        overlay_slot:
                                                                                                            VerticalAlignment:=vertical_alignment.Center
                                                                                                            HorizontalAlignment:=horizontal_alignment.Center
                                                                                                            Padding:=margin{Left:=0.0, Top:=0.0, Right:=0.0, Bottom:=0.0}
                                                                                                            Widget:=overlay:
                                                                                                                Slots:=array:

                                                                                                                    #Rarity texture border, when the rarity slot is empty
                                                                                                                    #overlay_slot:
                                                                                                                        #VerticalAlignment:=vertical_alignment.Center
                                                                                                                        #HorizontalAlignment:=horizontal_alignment.Center
                                                                                                                        #Padding:=margin{Left:=0.0, Top:=0.0, Right:=0.0, Bottom:=0.0}
                                                                                                                        #Widget:=texture_block{DefaultImage:=Textures.General.T_GenericBorderSquare2, DefaultDesiredSize:=vector2{X:=30.0, Y:=30.0}, DefaultTint:=MakeColorFromHex("#2e7be388")}
                                                                                                                    
                                                                                                                    #Rarity texture
                                                                                                                    overlay_slot:
                                                                                                                        VerticalAlignment:=vertical_alignment.Center
                                                                                                                        HorizontalAlignment:=horizontal_alignment.Center
                                                                                                                        Padding:=margin{Left:=0.0, Top:=0.0, Right:=0.0, Bottom:=0.0}
                                                                                                                        Widget:=texture_block{DefaultImage:=Textures.General.T_GenericSquare, DefaultDesiredSize:=vector2{X:=20.0, Y:=20.0}, DefaultTint:=MakeColorFromHex("#5b1e79FF")}
                                                                                                                    
                                                                                                        #Bracket
                                                                                                        overlay_slot:
                                                                                                            VerticalAlignment:=vertical_alignment.Fill
                                                                                                            HorizontalAlignment:=horizontal_alignment.Fill
                                                                                                            Padding:=margin{Left:=0.0, Top:=0.0, Right:=0.0, Bottom:=0.0}
                                                                                                            Widget:=stack_box:
                                                                                                                Orientation:=orientation.Horizontal
                                                                                                                #two instances of bracket texture to represent the range of rarities
                                                                                                                Slots:=array:

                                                                                                                    #stack_box_slot:
                                                                                                                        #VerticalAlignment:=vertical_alignment.Center
                                                                                                                        #HorizontalAlignment:=horizontal_alignment.Left
                                                                                                                        #Distribution:=option{1.0}
                                                                                                                        #Padding:=margin{Left:=0.0, Top:=0.0, Right:=0.0, Bottom:=0.0}
                                                                                                                        #Widget:=texture_block{DefaultImage:=Textures.General.T_RarityBracketLeft, DefaultDesiredSize:=vector2{X:=18.0, Y:=36.0}, DefaultTint:=MakeColorFromHex("#FFFFFF44")}

                                                                                                                    #stack_box_slot:
                                                                                                                        #VerticalAlignment:=vertical_alignment.Center
                                                                                                                        #HorizontalAlignment:=horizontal_alignment.Right
                                                                                                                        #Distribution:=option{1.0}
                                                                                                                        #Padding:=margin{Left:=0.0, Top:=0.0, Right:=0.0, Bottom:=0.0}
                                                                                                                        #Widget:=texture_block{DefaultImage:=Textures.General.T_RarityBracketRight, DefaultDesiredSize:=vector2{X:=18.0, Y:=36.0}, DefaultTint:=MakeColorFromHex("#FFFFFF44")}

                                                                                            stack_box_slot:
                                                                                                VerticalAlignment:=vertical_alignment.Center
                                                                                                HorizontalAlignment:=horizontal_alignment.Center
                                                                                                Distribution:=option{1.0}
                                                                                                Padding:=margin{Left:=0.0, Top:=0.0, Right:=0.0, Bottom:=0.0}
                                                                                                Widget:=overlay:
                                                                                                    Slots:=array:
                                                                                                        #Makeshift size box
                                                                                                        overlay_slot:
                                                                                                            VerticalAlignment:=vertical_alignment.Center
                                                                                                            HorizontalAlignment:=horizontal_alignment.Center
                                                                                                            Padding:=margin{Left:=0.0, Top:=0.0, Right:=0.0, Bottom:=0.0}
                                                                                                            Widget:=texture_block{DefaultImage:=Textures.General.T_GenSquTrans, DefaultDesiredSize:=vector2{X:=30.0, Y:=40.0}, DefaultTint:=MakeColorFromHex("#0000FF08")}

                                                                                                        #Overlay for rarity border and texture
                                                                                                        overlay_slot:
                                                                                                            VerticalAlignment:=vertical_alignment.Center
                                                                                                            HorizontalAlignment:=horizontal_alignment.Center
                                                                                                            Padding:=margin{Left:=0.0, Top:=0.0, Right:=0.0, Bottom:=0.0}
                                                                                                            Widget:=overlay:
                                                                                                                Slots:=array:

                                                                                                                    #Rarity texture border, when the rarity slot is empty
                                                                                                                    #overlay_slot:
                                                                                                                        #VerticalAlignment:=vertical_alignment.Center
                                                                                                                        #HorizontalAlignment:=horizontal_alignment.Center
                                                                                                                        #Padding:=margin{Left:=0.0, Top:=0.0, Right:=0.0, Bottom:=0.0}
                                                                                                                        #Widget:=texture_block{DefaultImage:=Textures.General.T_GenericBorderSquare2, DefaultDesiredSize:=vector2{X:=30.0, Y:=30.0}, DefaultTint:=MakeColorFromHex("#2e7be388")}
                                                                                                                    
                                                                                                                    #Rarity texture
                                                                                                                    overlay_slot:
                                                                                                                        VerticalAlignment:=vertical_alignment.Center
                                                                                                                        HorizontalAlignment:=horizontal_alignment.Center
                                                                                                                        Padding:=margin{Left:=0.0, Top:=0.0, Right:=0.0, Bottom:=0.0}
                                                                                                                        Widget:=texture_block{DefaultImage:=Textures.General.T_GenericSquare, DefaultDesiredSize:=vector2{X:=20.0, Y:=20.0}, DefaultTint:=MakeColorFromHex("#674f28FF")}

                                                                                                        #Bracket
                                                                                                        overlay_slot:
                                                                                                            VerticalAlignment:=vertical_alignment.Fill
                                                                                                            HorizontalAlignment:=horizontal_alignment.Fill
                                                                                                            Padding:=margin{Left:=0.0, Top:=0.0, Right:=0.0, Bottom:=0.0}
                                                                                                            Widget:=stack_box:
                                                                                                                Orientation:=orientation.Horizontal
                                                                                                                #two instances of bracket texture to represent the range of rarities
                                                                                                                Slots:=array:

                                                                                                                    #stack_box_slot:
                                                                                                                        #VerticalAlignment:=vertical_alignment.Center
                                                                                                                        #HorizontalAlignment:=horizontal_alignment.Left
                                                                                                                        #Distribution:=option{1.0}
                                                                                                                        #Padding:=margin{Left:=0.0, Top:=0.0, Right:=0.0, Bottom:=0.0}
                                                                                                                        #Widget:=texture_block{DefaultImage:=Textures.General.T_RarityBracketLeft, DefaultDesiredSize:=vector2{X:=18.0, Y:=36.0}, DefaultTint:=MakeColorFromHex("#FFFFFF44")}

                                                                                                                    #stack_box_slot:
                                                                                                                        #VerticalAlignment:=vertical_alignment.Center
                                                                                                                        #HorizontalAlignment:=horizontal_alignment.Right
                                                                                                                        #Distribution:=option{1.0}
                                                                                                                        #Padding:=margin{Left:=0.0, Top:=0.0, Right:=0.0, Bottom:=0.0}
                                                                                                                        #Widget:=texture_block{DefaultImage:=Textures.General.T_RarityBracketRight, DefaultDesiredSize:=vector2{X:=18.0, Y:=36.0}, DefaultTint:=MakeColorFromHex("#FFFFFF44")}

                                                                                            stack_box_slot:
                                                                                                VerticalAlignment:=vertical_alignment.Center
                                                                                                HorizontalAlignment:=horizontal_alignment.Center
                                                                                                Distribution:=option{1.0}
                                                                                                Padding:=margin{Left:=0.0, Top:=0.0, Right:=0.0, Bottom:=0.0}
                                                                                                Widget:=overlay:
                                                                                                    Slots:=array:
                                                                                                        #Makeshift size box
                                                                                                        overlay_slot:
                                                                                                            VerticalAlignment:=vertical_alignment.Center
                                                                                                            HorizontalAlignment:=horizontal_alignment.Center
                                                                                                            Padding:=margin{Left:=0.0, Top:=0.0, Right:=0.0, Bottom:=0.0}
                                                                                                            Widget:=texture_block{DefaultImage:=Textures.General.T_GenSquTrans, DefaultDesiredSize:=vector2{X:=30.0, Y:=40.0}, DefaultTint:=MakeColorFromHex("#0000FF08")}

                                                                                                        #Overlay for rarity border and texture
                                                                                                        overlay_slot:
                                                                                                            VerticalAlignment:=vertical_alignment.Center
                                                                                                            HorizontalAlignment:=horizontal_alignment.Center
                                                                                                            Padding:=margin{Left:=0.0, Top:=0.0, Right:=0.0, Bottom:=0.0}
                                                                                                            Widget:=overlay:
                                                                                                                Slots:=array:
                                                                                                        
                                                                                                                    #Rarity texture border, when the rarity slot is empty
                                                                                                                    #overlay_slot:
                                                                                                                        #VerticalAlignment:=vertical_alignment.Center
                                                                                                                        #HorizontalAlignment:=horizontal_alignment.Center
                                                                                                                        #Padding:=margin{Left:=0.0, Top:=0.0, Right:=0.0, Bottom:=0.0}
                                                                                                                        #Widget:=texture_block{DefaultImage:=Textures.General.T_GenericBorderSquare2, DefaultDesiredSize:=vector2{X:=30.0, Y:=30.0}, DefaultTint:=MakeColorFromHex("#2e7be388")}
                                                                                                                    
                                                                                                                    #Rarity texture
                                                                                                                    overlay_slot:
                                                                                                                        VerticalAlignment:=vertical_alignment.Center
                                                                                                                        HorizontalAlignment:=horizontal_alignment.Center
                                                                                                                        Padding:=margin{Left:=0.0, Top:=0.0, Right:=0.0, Bottom:=0.0}
                                                                                                                        Widget:=texture_block{DefaultImage:=Textures.General.T_GenericSquare, DefaultDesiredSize:=vector2{X:=20.0, Y:=20.0}, DefaultTint:=MakeColorFromHex("#78711fFF")}
                                                                                                                    
                                                                                                        #Bracket
                                                                                                        overlay_slot:
                                                                                                            VerticalAlignment:=vertical_alignment.Fill
                                                                                                            HorizontalAlignment:=horizontal_alignment.Fill
                                                                                                            Padding:=margin{Left:=0.0, Top:=0.0, Right:=0.0, Bottom:=0.0}
                                                                                                            Widget:=stack_box:
                                                                                                                Orientation:=orientation.Horizontal
                                                                                                                #two instances of bracket texture to represent the range of rarities
                                                                                                                Slots:=array:

                                                                                                                    #stack_box_slot:
                                                                                                                        #VerticalAlignment:=vertical_alignment.Center
                                                                                                                        #HorizontalAlignment:=horizontal_alignment.Left
                                                                                                                        #Distribution:=option{1.0}
                                                                                                                        #Padding:=margin{Left:=0.0, Top:=0.0, Right:=0.0, Bottom:=0.0}
                                                                                                                        #Widget:=texture_block{DefaultImage:=Textures.General.T_RarityBracketLeft, DefaultDesiredSize:=vector2{X:=18.0, Y:=36.0}, DefaultTint:=MakeColorFromHex("#FFFFFF44")}

                                                                                                                    stack_box_slot:
                                                                                                                        VerticalAlignment:=vertical_alignment.Center
                                                                                                                        HorizontalAlignment:=horizontal_alignment.Right
                                                                                                                        Distribution:=option{1.0}
                                                                                                                        Padding:=margin{Left:=0.0, Top:=0.0, Right:=0.0, Bottom:=0.0}
                                                                                                                        Widget:=texture_block{DefaultImage:=Textures.General.T_RarityBracketRight, DefaultDesiredSize:=vector2{X:=18.0, Y:=36.0}, DefaultTint:=MakeColorFromHex("#a3a6aaff")}

                    #overlay_slot:
                        #VerticalAlignment:=vertical_alignment.Fill
                        #HorizontalAlignment:=horizontal_alignment.Fill
                        #Padding:=margin{Left:=0.0, Top:=0.0, Right:=0.0, Bottom:=0.0}
                        #Widget:=texture_block{DefaultImage:=Textures.General.T_LevelPointer, DefaultDesiredSize:=vector2{X:=20.0, Y:=20.0}, DefaultTint:=MakeColorFromHex("#7de2FFFF")}

            #Definition for the Sixth level row
            LevelRow6:=overlay:
                Slots:=array:
                    overlay_slot:
                        #Makeshift size box
                        VerticalAlignment:=vertical_alignment.Center
                        HorizontalAlignment:=horizontal_alignment.Center
                        Padding:=margin{Left:=0.0, Top:=0.0, Right:=0.0, Bottom:=0.0}
                        Widget:=texture_block{DefaultImage:=Textures.General.T_GenSquTrans, DefaultDesiredSize:=vector2{X:=1250.0, Y:=114.0}, DefaultTint:=MakeColorFromHex("#0000FFFF")}

                    overlay_slot:
                        #Rarity level content
                        VerticalAlignment:=vertical_alignment.Center
                        HorizontalAlignment:=horizontal_alignment.Left
                        Padding:=margin{Left:=0.0, Top:=0.0, Right:=0.0, Bottom:=0.0}
                        Widget:=overlay:
                            Slots:=array:

                                overlay_slot:
                                    #Makeshift size box
                                    VerticalAlignment:=vertical_alignment.Center
                                    HorizontalAlignment:=horizontal_alignment.Center
                                    Padding:=margin{Left:=0.0, Top:=0.0, Right:=0.0, Bottom:=0.0}
                                    Widget:=texture_block{DefaultImage:=Textures.General.T_GenericSquare, DefaultDesiredSize:=vector2{X:=1190.0, Y:=114.0}, DefaultTint:=MakeColorFromHex("#010a1832")}

                                overlay_slot:
                                    #Highlight texture, over the content but under the rarity flag
                                    VerticalAlignment:=vertical_alignment.Center
                                    HorizontalAlignment:=horizontal_alignment.Center
                                    Padding:=margin{Left:=0.0, Top:=0.0, Right:=0.0, Bottom:=0.0}
                                    Widget:=texture_block{DefaultImage:=Textures.General.T_LevelBorder, DefaultDesiredSize:=vector2{X:=1190.0, Y:=114.0}, DefaultTint:=MakeColorFromHex("#174583FF")}

                                    #Horizontal stack box
                                overlay_slot:
                                    VerticalAlignment:=vertical_alignment.Fill
                                    HorizontalAlignment:=horizontal_alignment.Left
                                    Padding:=margin{Left:=0.0, Top:=0.0, Right:=0.0, Bottom:=0.0}
                                    Widget:=stack_box:
                                        Orientation:=orientation.Horizontal
                                        Slots:=array:
                                            #Level flag
                                            stack_box_slot:
                                                VerticalAlignment:=vertical_alignment.Top
                                                HorizontalAlignment:=horizontal_alignment.Left
                                                Padding:=margin{Left:=0.0, Top:=0.0, Right:=15.0, Bottom:=0.0}
                                                Widget:=overlay:

                                                    Slots:=array:
                                                        overlay_slot:
                                                            VerticalAlignment:=vertical_alignment.Center
                                                            HorizontalAlignment:=horizontal_alignment.Center
                                                            Padding:=margin{Left:=0.0, Top:=0.0, Right:=0.0, Bottom:=0.0}
                                                            #Tier level makeshift size box
                                                            Widget:=texture_block{DefaultImage:=Textures.General.T_GenSquTrans, DefaultDesiredSize:=vector2{X:=150.0, Y:=114.0}, DefaultTint:=MakeColorFromHex("#FF0000FF")}

                                                        overlay_slot:
                                                            VerticalAlignment:=vertical_alignment.Top
                                                            HorizontalAlignment:=horizontal_alignment.Left
                                                            Padding:=margin{Left:=0.0, Top:=0.0, Right:=0.0, Bottom:=0.0}
                                                            #Selected level yellow flag
                                                            Widget:=overlay:

                                                                #Slots:=array:
                                                                    #overlay_slot:
                                                                        #VerticalAlignment:=vertical_alignment.Center
                                                                        #HorizontalAlignment:=horizontal_alignment.Center
                                                                        #Padding:=margin{Left:=0.0, Top:=0.0, Right:=0.0, Bottom:=0.0}
                                                                        #level flag texture
                                                                        #Widget:=texture_block{DefaultImage:=Textures.General.T_GenLevelFlag, DefaultDesiredSize:=vector2{X:=140.0, Y:=45.0}, DefaultTint:=MakeColorFromHex("#f7ff1aFF")}

                                                                    #overlay_slot:
                                                                        #VerticalAlignment:=vertical_alignment.Center
                                                                        #HorizontalAlignment:=horizontal_alignment.Center
                                                                        #Padding:=margin{Left:=0.0, Top:=0.0, Right:=0.0, Bottom:=0.0}
                                                                        #level flag text
                                                                        #Widget:=text_block{DefaultText := StringToMessage("Level 0"), DefaultTextColor:=MakeColorFromHex("#000000FF")}

                                                        overlay_slot:
                                                            VerticalAlignment:=vertical_alignment.Fill
                                                            HorizontalAlignment:=horizontal_alignment.Center
                                                            Padding:=margin{Left:=0.0, Top:=5.0, Right:=0.0, Bottom:=5.0}
                                                            #Unselected level and lock
                                                            Widget:=overlay:
                                                                Slots:=array:
                                                                    
                                                                    overlay_slot:
                                                                        VerticalAlignment:=vertical_alignment.Top
                                                                        HorizontalAlignment:=horizontal_alignment.Center
                                                                        Padding:=margin{Left:=0.0, Top:=0.0, Right:=0.0, Bottom:=0.0}
                                                                        #level flag texture
                                                                        Widget:=text_block{DefaultText := StringToMessage("Level 6"), DefaultTextColor:=MakeColorFromHex("#a3a6aaff")}

                                                                    overlay_slot:
                                                                        VerticalAlignment:=vertical_alignment.Bottom
                                                                        HorizontalAlignment:=horizontal_alignment.Center
                                                                        Padding:=margin{Left:=0.0, Top:=0.0, Right:=0.0, Bottom:=0.0}
                                                                        #level flag texture
                                                                        Widget:=texture_block{DefaultImage:=Textures.General.T_Lock, DefaultDesiredSize:=vector2{X:=52.0, Y:=58.0}, DefaultTint:=MakeColorFromHex("#a3a6aaff")}
                                            
                                            #Level weapon roster
                                            stack_box_slot:
                                                VerticalAlignment:=vertical_alignment.Center
                                                HorizontalAlignment:=horizontal_alignment.Center
                                                Padding:=margin{Left:=0.0, Top:=0.0, Right:=8.0, Bottom:=0.0}
                                                Widget:=overlay:
                                                    Slots:=array:
                                                        #Weapon Icon - Pistols
                                                        overlay_slot:
                                                            VerticalAlignment:=vertical_alignment.Center
                                                            HorizontalAlignment:=horizontal_alignment.Center
                                                            Padding:=margin{Left:=0.0, Top:=0.0, Right:=0.0, Bottom:=0.0}
                                                            #Weapon icon BG
                                                            Widget:=texture_block{DefaultImage:=Textures.General.T_GenericSquare, DefaultDesiredSize:=vector2{X:=78.0, Y:=78.0}, DefaultTint:=MakeColorFromHex("#000c1d88")}

                                                        overlay_slot:
                                                            VerticalAlignment:=vertical_alignment.Center
                                                            HorizontalAlignment:=horizontal_alignment.Center
                                                            Padding:=margin{Left:=0.0, Top:=0.0, Right:=0.0, Bottom:=0.0}
                                                            #Weapon icon Border
                                                            Widget:=texture_block{DefaultImage:=Textures.General.T_GenericBorderSquare, DefaultDesiredSize:=vector2{X:=78.0, Y:=78.0}, DefaultTint:=MakeColorFromHex("#174583FF")}

                                                        overlay_slot:
                                                            VerticalAlignment:=vertical_alignment.Top
                                                            HorizontalAlignment:=horizontal_alignment.Center
                                                            Padding:=margin{Left:=0.0, Top:=5.0, Right:=0.0, Bottom:=0.0}
                                                            #Weapon icon - pistol
                                                            Widget:=texture_block{DefaultImage:=Textures.Items.T_IconPistol, DefaultDesiredSize:=vector2{X:=50.0, Y:=50.0}, DefaultTint:=MakeColorFromHex("#a3a6aaff")}

                                                        overlay_slot:
                                                            VerticalAlignment:=vertical_alignment.Bottom
                                                            HorizontalAlignment:=horizontal_alignment.Center
                                                            Padding:=margin{Left:=0.0, Top:=0.0, Right:=0.0, Bottom:=5.0}
                                                            #Weapon header - pistol
                                                            Widget:=texture_block{DefaultImage:=Textures.Items.T_HeaderPistol, DefaultDesiredSize:=vector2{X:=78.0, Y:=18.0}, DefaultTint:=MakeColorFromHex("#a3a6aaff")}
                                                        
                                            stack_box_slot:
                                                VerticalAlignment:=vertical_alignment.Center
                                                HorizontalAlignment:=horizontal_alignment.Center
                                                Padding:=margin{Left:=0.0, Top:=0.0, Right:=8.0, Bottom:=0.0}
                                                Widget:=overlay:
                                                    Slots:=array:
                                                        #Weapon Icon - Shotgun
                                                        overlay_slot:
                                                            VerticalAlignment:=vertical_alignment.Center
                                                            HorizontalAlignment:=horizontal_alignment.Center
                                                            Padding:=margin{Left:=0.0, Top:=0.0, Right:=0.0, Bottom:=0.0}
                                                            #Weapon icon BG
                                                            Widget:=texture_block{DefaultImage:=Textures.General.T_GenericSquare, DefaultDesiredSize:=vector2{X:=78.0, Y:=78.0}, DefaultTint:=MakeColorFromHex("#000c1d88")}

                                                        overlay_slot:
                                                            VerticalAlignment:=vertical_alignment.Center
                                                            HorizontalAlignment:=horizontal_alignment.Center
                                                            Padding:=margin{Left:=0.0, Top:=0.0, Right:=0.0, Bottom:=0.0}
                                                            #Weapon icon Border
                                                            Widget:=texture_block{DefaultImage:=Textures.General.T_GenericBorderSquare, DefaultDesiredSize:=vector2{X:=78.0, Y:=78.0}, DefaultTint:=MakeColorFromHex("#174583FF")}

                                                        overlay_slot:
                                                            VerticalAlignment:=vertical_alignment.Top
                                                            HorizontalAlignment:=horizontal_alignment.Center
                                                            Padding:=margin{Left:=0.0, Top:=5.0, Right:=0.0, Bottom:=0.0}
                                                            #Weapon icon - shotgun
                                                            Widget:=texture_block{DefaultImage:=Textures.Items.T_IconShotgun, DefaultDesiredSize:=vector2{X:=50.0, Y:=50.0}, DefaultTint:=MakeColorFromHex("#a3a6aaff")}

                                                        overlay_slot:
                                                            VerticalAlignment:=vertical_alignment.Bottom
                                                            HorizontalAlignment:=horizontal_alignment.Center
                                                            Padding:=margin{Left:=0.0, Top:=0.0, Right:=0.0, Bottom:=5.0}
                                                            #Weapon header - shotgun
                                                            Widget:=texture_block{DefaultImage:=Textures.Items.T_HeaderShotgun, DefaultDesiredSize:=vector2{X:=78.0, Y:=18.0}, DefaultTint:=MakeColorFromHex("#a3a6aaff")}

                                            stack_box_slot:
                                                VerticalAlignment:=vertical_alignment.Center
                                                HorizontalAlignment:=horizontal_alignment.Center
                                                Padding:=margin{Left:=0.0, Top:=0.0, Right:=8.0, Bottom:=0.0}
                                                Widget:=overlay:
                                                    Slots:=array:
                                                        #Weapon Icon - SMG
                                                        overlay_slot:
                                                            VerticalAlignment:=vertical_alignment.Top
                                                            HorizontalAlignment:=horizontal_alignment.Center
                                                            Padding:=margin{Left:=0.0, Top:=0.0, Right:=0.0, Bottom:=0.0}
                                                            #Weapon icon BG
                                                            Widget:=texture_block{DefaultImage:=Textures.General.T_GenericSquare, DefaultDesiredSize:=vector2{X:=78.0, Y:=78.0}, DefaultTint:=MakeColorFromHex("#000c1d88")}

                                                        overlay_slot:
                                                            VerticalAlignment:=vertical_alignment.Center
                                                            HorizontalAlignment:=horizontal_alignment.Center
                                                            Padding:=margin{Left:=0.0, Top:=0.0, Right:=0.0, Bottom:=0.0}
                                                            #Weapon icon Border
                                                            Widget:=texture_block{DefaultImage:=Textures.General.T_GenericBorderSquare, DefaultDesiredSize:=vector2{X:=78.0, Y:=78.0}, DefaultTint:=MakeColorFromHex("#174583FF")}

                                                        overlay_slot:
                                                            VerticalAlignment:=vertical_alignment.Top
                                                            HorizontalAlignment:=horizontal_alignment.Center
                                                            Padding:=margin{Left:=0.0, Top:=5.0, Right:=0.0, Bottom:=0.0}
                                                            #Weapon icon - SMG
                                                            Widget:=texture_block{DefaultImage:=Textures.Items.T_IconSMG, DefaultDesiredSize:=vector2{X:=50.0, Y:=50.0}, DefaultTint:=MakeColorFromHex("#a3a6aaff")}

                                                        overlay_slot:
                                                            VerticalAlignment:=vertical_alignment.Bottom
                                                            HorizontalAlignment:=horizontal_alignment.Center
                                                            Padding:=margin{Left:=0.0, Top:=0.0, Right:=0.0, Bottom:=5.0}
                                                            #Weapon header - SMG
                                                            Widget:=texture_block{DefaultImage:=Textures.Items.T_HeaderSMGs, DefaultDesiredSize:=vector2{X:=78.0, Y:=18.0}, DefaultTint:=MakeColorFromHex("#a3a6aaff")}
                                                            

                                            stack_box_slot:
                                                VerticalAlignment:=vertical_alignment.Center
                                                HorizontalAlignment:=horizontal_alignment.Center
                                                Padding:=margin{Left:=0.0, Top:=0.0, Right:=8.0, Bottom:=0.0}
                                                Widget:=overlay:
                                                    Slots:=array:
                                                        #Weapon Icon - Rifle
                                                        overlay_slot:
                                                            VerticalAlignment:=vertical_alignment.Center
                                                            HorizontalAlignment:=horizontal_alignment.Center
                                                            Padding:=margin{Left:=0.0, Top:=0.0, Right:=0.0, Bottom:=0.0}
                                                            #Weapon icon BG
                                                            Widget:=texture_block{DefaultImage:=Textures.General.T_GenericSquare, DefaultDesiredSize:=vector2{X:=78.0, Y:=78.0}, DefaultTint:=MakeColorFromHex("#000c1d88")}

                                                        overlay_slot:
                                                            VerticalAlignment:=vertical_alignment.Top
                                                            HorizontalAlignment:=horizontal_alignment.Center
                                                            Padding:=margin{Left:=0.0, Top:=0.0, Right:=0.0, Bottom:=0.0}
                                                            #Weapon icon Border
                                                            Widget:=texture_block{DefaultImage:=Textures.General.T_GenericBorderSquare, DefaultDesiredSize:=vector2{X:=78.0, Y:=78.0}, DefaultTint:=MakeColorFromHex("#174583FF")}

                                                        overlay_slot:
                                                            VerticalAlignment:=vertical_alignment.Top
                                                            HorizontalAlignment:=horizontal_alignment.Center
                                                            Padding:=margin{Left:=0.0, Top:=5.0, Right:=0.0, Bottom:=0.0}
                                                            #Weapon icon - Rifle
                                                            Widget:=texture_block{DefaultImage:=Textures.Items.T_IconRifle, DefaultDesiredSize:=vector2{X:=50.0, Y:=50.0}, DefaultTint:=MakeColorFromHex("#a3a6aaff")}

                                                        overlay_slot:
                                                            VerticalAlignment:=vertical_alignment.Bottom
                                                            HorizontalAlignment:=horizontal_alignment.Center
                                                            Padding:=margin{Left:=0.0, Top:=0.0, Right:=0.0, Bottom:=5.0}
                                                            #Weapon header - Rifle
                                                            Widget:=texture_block{DefaultImage:=Textures.Items.T_HeaderRifle, DefaultDesiredSize:=vector2{X:=78.0, Y:=18.0}, DefaultTint:=MakeColorFromHex("#a3a6aaff")}

                                            stack_box_slot:
                                                VerticalAlignment:=vertical_alignment.Center
                                                HorizontalAlignment:=horizontal_alignment.Center
                                                Padding:=margin{Left:=0.0, Top:=0.0, Right:=8.0, Bottom:=0.0}
                                                Widget:=overlay:
                                                    Slots:=array:
                                                        #Weapon Icon - Explosive
                                                        overlay_slot:
                                                            VerticalAlignment:=vertical_alignment.Center
                                                            HorizontalAlignment:=horizontal_alignment.Center
                                                            Padding:=margin{Left:=0.0, Top:=0.0, Right:=0.0, Bottom:=0.0}
                                                            #Weapon icon BG
                                                            Widget:=texture_block{DefaultImage:=Textures.General.T_GenericSquare, DefaultDesiredSize:=vector2{X:=78.0, Y:=78.0}, DefaultTint:=MakeColorFromHex("#000c1d88")}

                                                        overlay_slot:
                                                            VerticalAlignment:=vertical_alignment.Center
                                                            HorizontalAlignment:=horizontal_alignment.Center
                                                            Padding:=margin{Left:=0.0, Top:=0.0, Right:=0.0, Bottom:=0.0}
                                                            #Weapon icon Border
                                                            Widget:=texture_block{DefaultImage:=Textures.General.T_GenericBorderSquare, DefaultDesiredSize:=vector2{X:=78.0, Y:=78.0}, DefaultTint:=MakeColorFromHex("#174583FF")}

                                                        overlay_slot:
                                                            VerticalAlignment:=vertical_alignment.Top
                                                            HorizontalAlignment:=horizontal_alignment.Center
                                                            Padding:=margin{Left:=0.0, Top:=5.0, Right:=0.0, Bottom:=0.0}
                                                            #Weapon icon - Explosive
                                                            Widget:=texture_block{DefaultImage:=Textures.Items.T_IconExplosive, DefaultDesiredSize:=vector2{X:=50.0, Y:=50.0}, DefaultTint:=MakeColorFromHex("#a3a6aaff")}

                                                        overlay_slot:
                                                            VerticalAlignment:=vertical_alignment.Bottom
                                                            HorizontalAlignment:=horizontal_alignment.Center
                                                            Padding:=margin{Left:=0.0, Top:=0.0, Right:=0.0, Bottom:=5.0}
                                                            #Weapon header - Explosive
                                                            Widget:=texture_block{DefaultImage:=Textures.Items.T_HeaderExplosive, DefaultDesiredSize:=vector2{X:=78.0, Y:=18.0}, DefaultTint:=MakeColorFromHex("#a3a6aaff")}

                                            stack_box_slot:
                                                VerticalAlignment:=vertical_alignment.Center
                                                HorizontalAlignment:=horizontal_alignment.Center
                                                Padding:=margin{Left:=0.0, Top:=0.0, Right:=8.0, Bottom:=0.0}
                                                Widget:=overlay:
                                                    Slots:=array:
                                                        #Weapon Icon - Special
                                                        overlay_slot:
                                                            VerticalAlignment:=vertical_alignment.Center
                                                            HorizontalAlignment:=horizontal_alignment.Center
                                                            Padding:=margin{Left:=0.0, Top:=0.0, Right:=0.0, Bottom:=0.0}
                                                            #Weapon icon BG
                                                            Widget:=texture_block{DefaultImage:=Textures.General.T_GenericSquare, DefaultDesiredSize:=vector2{X:=78.0, Y:=78.0}, DefaultTint:=MakeColorFromHex("#000c1d88")}

                                                        overlay_slot:
                                                            VerticalAlignment:=vertical_alignment.Center
                                                            HorizontalAlignment:=horizontal_alignment.Center
                                                            Padding:=margin{Left:=0.0, Top:=0.0, Right:=0.0, Bottom:=0.0}
                                                            #Weapon icon Border
                                                            Widget:=texture_block{DefaultImage:=Textures.General.T_GenericBorderSquare, DefaultDesiredSize:=vector2{X:=78.0, Y:=78.0}, DefaultTint:=MakeColorFromHex("#174583FF")}

                                                        overlay_slot:
                                                            VerticalAlignment:=vertical_alignment.Top
                                                            HorizontalAlignment:=horizontal_alignment.Center
                                                            Padding:=margin{Left:=0.0, Top:=5.0, Right:=0.0, Bottom:=0.0}
                                                            #Weapon icon - special
                                                            Widget:=texture_block{DefaultImage:=Textures.Items.T_IconArtifact, DefaultDesiredSize:=vector2{X:=50.0, Y:=50.0}, DefaultTint:=MakeColorFromHex("#a3a6aaff")}

                                                        overlay_slot:
                                                            VerticalAlignment:=vertical_alignment.Bottom
                                                            HorizontalAlignment:=horizontal_alignment.Center
                                                            Padding:=margin{Left:=0.0, Top:=0.0, Right:=0.0, Bottom:=5.0}
                                                            #Weapon header - Special
                                                            Widget:=texture_block{DefaultImage:=Textures.Items.T_HeaderSpecial, DefaultDesiredSize:=vector2{X:=78.0, Y:=18.0}, DefaultTint:=MakeColorFromHex("#a3a6aaff")}

                                            stack_box_slot:
                                                VerticalAlignment:=vertical_alignment.Center
                                                HorizontalAlignment:=horizontal_alignment.Center
                                                Padding:=margin{Left:=0.0, Top:=0.0, Right:=8.0, Bottom:=0.0}
                                                Widget:=overlay:
                                                    Slots:=array:
                                                        #Weapon Icon - DMR Sniper
                                                        overlay_slot:
                                                            VerticalAlignment:=vertical_alignment.Center
                                                            HorizontalAlignment:=horizontal_alignment.Center
                                                            Padding:=margin{Left:=0.0, Top:=0.0, Right:=0.0, Bottom:=0.0}
                                                            #Weapon icon BG
                                                            Widget:=texture_block{DefaultImage:=Textures.General.T_GenericSquare, DefaultDesiredSize:=vector2{X:=78.0, Y:=78.0}, DefaultTint:=MakeColorFromHex("#000c1d88")}

                                                        overlay_slot:
                                                            VerticalAlignment:=vertical_alignment.Center
                                                            HorizontalAlignment:=horizontal_alignment.Center
                                                            Padding:=margin{Left:=0.0, Top:=0.0, Right:=0.0, Bottom:=0.0}
                                                            #Weapon icon Border
                                                            Widget:=texture_block{DefaultImage:=Textures.General.T_GenericBorderSquare, DefaultDesiredSize:=vector2{X:=78.0, Y:=78.0}, DefaultTint:=MakeColorFromHex("#174583FF")}

                                                        overlay_slot:
                                                            VerticalAlignment:=vertical_alignment.Top
                                                            HorizontalAlignment:=horizontal_alignment.Center
                                                            Padding:=margin{Left:=0.0, Top:=5.0, Right:=0.0, Bottom:=0.0}
                                                            #Weapon icon - Sniper
                                                            Widget:=texture_block{DefaultImage:=Textures.Items.T_IconSniper, DefaultDesiredSize:=vector2{X:=50.0, Y:=50.0}, DefaultTint:=MakeColorFromHex("#a3a6aaff")}

                                                        overlay_slot:
                                                            VerticalAlignment:=vertical_alignment.Bottom
                                                            HorizontalAlignment:=horizontal_alignment.Center
                                                            Padding:=margin{Left:=0.0, Top:=0.0, Right:=0.0, Bottom:=5.0}
                                                            #Weapon header - DMR Sniper
                                                            Widget:=texture_block{DefaultImage:=Textures.Items.T_HeaderDMRSniper, DefaultDesiredSize:=vector2{X:=78.0, Y:=18.0}, DefaultTint:=MakeColorFromHex("#a3a6aaff")}

                                            stack_box_slot:
                                                VerticalAlignment:=vertical_alignment.Fill
                                                HorizontalAlignment:=horizontal_alignment.Center #Center
                                                Padding:=margin{Left:=0.0, Top:=0.0, Right:=0.0, Bottom:=0.0}
                                                Widget:=overlay:
                                                    Slots:=array:
                                                        overlay_slot:
                                                            #Makeshift size box
                                                            VerticalAlignment:=vertical_alignment.Center
                                                            HorizontalAlignment:=horizontal_alignment.Center
                                                            Padding:=margin{Left:=0.0, Top:=0.0, Right:=0.0, Bottom:=0.0}
                                                            Widget:=texture_block{DefaultImage:=Textures.General.T_GenSquTrans, DefaultDesiredSize:=vector2{X:=418.0, Y:=100.0}, DefaultTint:=MakeColorFromHex("#00FF0011")}

                                                        overlay_slot:
                                                            #Vertical stack box to add all the rarity information
                                                            VerticalAlignment:=vertical_alignment.Center
                                                            HorizontalAlignment:=horizontal_alignment.Center
                                                            Padding:=margin{Left:=0.0, Top:=0.0, Right:=0.0, Bottom:=0.0}
                                                            Widget:=stack_box:
                                                                Orientation:=orientation.Vertical
                                                                Slots:=array:
                                                                    stack_box_slot:
                                                                        #Horizontal Stack box for max rarity available
                                                                        VerticalAlignment:=vertical_alignment.Top
                                                                        HorizontalAlignment:=horizontal_alignment.Center
                                                                        Padding:=margin{Left:=0.0, Top:=0.0, Right:=0.0, Bottom:=0.0}
                                                                        Widget:=stack_box:
                                                                            Orientation:=orientation.Horizontal
                                                                            Slots:=array:
                                                                                stack_box_slot:
                                                                                    VerticalAlignment:=vertical_alignment.Center
                                                                                    HorizontalAlignment:=horizontal_alignment.Center
                                                                                    Padding:=margin{Left:=0.0, Top:=0.0, Right:=0.0, Bottom:=0.0}
                                                                                    Widget:=text_block{DefaultText := StringToMessage("Rarity chance up to "), DefaultTextColor:=MakeColorFromHex("#a3a6aaff")}

                                                                                stack_box_slot:
                                                                                    VerticalAlignment:=vertical_alignment.Center
                                                                                    HorizontalAlignment:=horizontal_alignment.Center
                                                                                    Padding:=margin{Left:=0.0, Top:=0.0, Right:=0.0, Bottom:=0.0}
                                                                                    Widget:=texture_block{DefaultImage:=Textures.Rarity.T_FlagMythic, DefaultDesiredSize:=vector2{X:=104.0, Y:=36.0}, DefaultTint:=MakeColorFromHex("#a3a6aaff")}

                                                                    stack_box_slot:
                                                                        #Horizontal overlay for overall rarity scheme
                                                                        VerticalAlignment:=vertical_alignment.Bottom
                                                                        HorizontalAlignment:=horizontal_alignment.Center
                                                                        Padding:=margin{Left:=0.0, Top:=0.0, Right:=0.0, Bottom:=0.0}
                                                                        Widget:=overlay:
                                                                            Slots:=array:
                                                                                
                                                                                #Makeshift size box for the rarity scheme
                                                                                overlay_slot:
                                                                                    VerticalAlignment:=vertical_alignment.Center
                                                                                    HorizontalAlignment:=horizontal_alignment.Center
                                                                                    Padding:=margin{Left:=0.0, Top:=0.0, Right:=0.0, Bottom:=0.0}
                                                                                    Widget:=texture_block{DefaultImage:=Textures.General.T_GenSquTrans, DefaultDesiredSize:=vector2{X:=255.0, Y:=50.0}, DefaultTint:=MakeColorFromHex("#FF000022")}#Red frame

                                                                                #Visible border for rarity layout
                                                                                overlay_slot:
                                                                                    VerticalAlignment:=vertical_alignment.Center
                                                                                    HorizontalAlignment:=horizontal_alignment.Center
                                                                                    Padding:=margin{Left:=0.0, Top:=0.0, Right:=0.0, Bottom:=0.0}
                                                                                    Widget:=texture_block{DefaultImage:=Textures.General.T_GenericBorderSquare3, DefaultDesiredSize:=vector2{X:=255.0, Y:=40.0}, DefaultTint:=MakeColorFromHex("#174583FF")}

                                                                                #Horizontal Stack box for rarity slots
                                                                                overlay_slot:
                                                                                    VerticalAlignment:=vertical_alignment.Fill
                                                                                    HorizontalAlignment:=horizontal_alignment.Fill
                                                                                    Padding:=margin{Left:=5.0, Top:=0.0, Right:=5.0, Bottom:=0.0}
                                                                                    Widget:=stack_box:
                                                                                        Orientation:=orientation.Horizontal
                                                                                        Slots:=array:
                                                                                            #Overlay for each rarity slots, including the pointing arrow
                                                                                            stack_box_slot:
                                                                                                VerticalAlignment:=vertical_alignment.Center
                                                                                                HorizontalAlignment:=horizontal_alignment.Center
                                                                                                Distribution:=option{1.0}
                                                                                                Padding:=margin{Left:=0.0, Top:=0.0, Right:=0.0, Bottom:=0.0}
                                                                                                Widget:=overlay:
                                                                                                    Slots:=array:
                                                                                                        #Makeshift size box
                                                                                                        overlay_slot:
                                                                                                            VerticalAlignment:=vertical_alignment.Center
                                                                                                            HorizontalAlignment:=horizontal_alignment.Center
                                                                                                            Padding:=margin{Left:=0.0, Top:=0.0, Right:=0.0, Bottom:=0.0}
                                                                                                            Widget:=texture_block{DefaultImage:=Textures.General.T_GenSquTrans, DefaultDesiredSize:=vector2{X:=30.0, Y:=40.0}, DefaultTint:=MakeColorFromHex("#0000FF08")}

                                                                                                        #Overlay for rarity border and texture
                                                                                                        overlay_slot:
                                                                                                            VerticalAlignment:=vertical_alignment.Center
                                                                                                            HorizontalAlignment:=horizontal_alignment.Center
                                                                                                            Padding:=margin{Left:=0.0, Top:=0.0, Right:=0.0, Bottom:=0.0}
                                                                                                            Widget:=overlay:
                                                                                                                Slots:=array:
                                                                                                                    
                                                                                                                    #Rarity texture border, when the rarity slot is empty
                                                                                                                    overlay_slot:
                                                                                                                        VerticalAlignment:=vertical_alignment.Center
                                                                                                                        HorizontalAlignment:=horizontal_alignment.Center
                                                                                                                        Padding:=margin{Left:=0.0, Top:=0.0, Right:=0.0, Bottom:=0.0}
                                                                                                                        Widget:=texture_block{DefaultImage:=Textures.General.T_GenericBorderSquare2, DefaultDesiredSize:=vector2{X:=30.0, Y:=30.0}, DefaultTint:=MakeColorFromHex("#174583FF")}

                                                                                                                    #Rarity texture
                                                                                                                    #overlay_slot:
                                                                                                                        #VerticalAlignment:=vertical_alignment.Center
                                                                                                                        #HorizontalAlignment:=horizontal_alignment.Center
                                                                                                                        #Padding:=margin{Left:=0.0, Top:=0.0, Right:=0.0, Bottom:=0.0}
                                                                                                                        #Widget:=texture_block{DefaultImage:=Textures.General.T_GenericSquare, DefaultDesiredSize:=vector2{X:=30.0, Y:=30.0}, DefaultTint:=MakeColorFromHex("#979797FF")}
                                                                                                        
                                                                                                                    

                                                                                                        #Bracket
                                                                                                        overlay_slot:
                                                                                                            VerticalAlignment:=vertical_alignment.Fill
                                                                                                            HorizontalAlignment:=horizontal_alignment.Fill
                                                                                                            Padding:=margin{Left:=0.0, Top:=0.0, Right:=0.0, Bottom:=0.0}
                                                                                                            Widget:=stack_box:
                                                                                                                Orientation:=orientation.Horizontal
                                                                                                                #two instances of bracket texture to represent the range of rarities
                                                                                                                Slots:=array:

                                                                                                                    #stack_box_slot:
                                                                                                                        #VerticalAlignment:=vertical_alignment.Center
                                                                                                                        #HorizontalAlignment:=horizontal_alignment.Left
                                                                                                                        #Distribution:=option{1.0}
                                                                                                                        #Padding:=margin{Left:=0.0, Top:=0.0, Right:=0.0, Bottom:=0.0}
                                                                                                                        #Widget:=texture_block{DefaultImage:=Textures.General.T_RarityBracketLeft, DefaultDesiredSize:=vector2{X:=18.0, Y:=36.0}, DefaultTint:=MakeColorFromHex("#FFFFFF44")}

                                                                                                                    #stack_box_slot:
                                                                                                                        #VerticalAlignment:=vertical_alignment.Center
                                                                                                                        #HorizontalAlignment:=horizontal_alignment.Right
                                                                                                                        #Distribution:=option{1.0}
                                                                                                                        #Padding:=margin{Left:=0.0, Top:=0.0, Right:=0.0, Bottom:=0.0}
                                                                                                                        #Widget:=texture_block{DefaultImage:=Textures.General.T_RarityBracketRight, DefaultDesiredSize:=vector2{X:=18.0, Y:=36.0}, DefaultTint:=MakeColorFromHex("#FFFFFF44")}

                                                                                            stack_box_slot:
                                                                                                VerticalAlignment:=vertical_alignment.Center
                                                                                                HorizontalAlignment:=horizontal_alignment.Center
                                                                                                Distribution:=option{1.0}
                                                                                                Padding:=margin{Left:=0.0, Top:=0.0, Right:=0.0, Bottom:=0.0}
                                                                                                Widget:=overlay:
                                                                                                    Slots:=array:
                                                                                                        #Makeshift size box
                                                                                                        overlay_slot:
                                                                                                            VerticalAlignment:=vertical_alignment.Center
                                                                                                            HorizontalAlignment:=horizontal_alignment.Center
                                                                                                            Padding:=margin{Left:=0.0, Top:=0.0, Right:=0.0, Bottom:=0.0}
                                                                                                            Widget:=texture_block{DefaultImage:=Textures.General.T_GenSquTrans, DefaultDesiredSize:=vector2{X:=30.0, Y:=40.0}, DefaultTint:=MakeColorFromHex("#0000FF08")}

                                                                                                        #Overlay for rarity border and texture
                                                                                                        overlay_slot:
                                                                                                            VerticalAlignment:=vertical_alignment.Center
                                                                                                            HorizontalAlignment:=horizontal_alignment.Center
                                                                                                            Padding:=margin{Left:=0.0, Top:=0.0, Right:=0.0, Bottom:=0.0}
                                                                                                            Widget:=overlay:
                                                                                                                Slots:=array:
                                                                                                                    
                                                                                                                    #Rarity texture border, when the rarity slot is empty
                                                                                                                    overlay_slot:
                                                                                                                        VerticalAlignment:=vertical_alignment.Center
                                                                                                                        HorizontalAlignment:=horizontal_alignment.Center
                                                                                                                        Padding:=margin{Left:=0.0, Top:=0.0, Right:=0.0, Bottom:=0.0}
                                                                                                                        Widget:=texture_block{DefaultImage:=Textures.General.T_GenericBorderSquare2, DefaultDesiredSize:=vector2{X:=30.0, Y:=30.0}, DefaultTint:=MakeColorFromHex("#174583FF")}

                                                                                                                    #Rarity texture
                                                                                                                    #overlay_slot:
                                                                                                                        #VerticalAlignment:=vertical_alignment.Center
                                                                                                                        #HorizontalAlignment:=horizontal_alignment.Center
                                                                                                                        #Padding:=margin{Left:=0.0, Top:=0.0, Right:=0.0, Bottom:=0.0}
                                                                                                                        #Widget:=texture_block{DefaultImage:=Textures.General.T_GenericSquare, DefaultDesiredSize:=vector2{X:=30.0, Y:=30.0}, DefaultTint:=MakeColorFromHex("#4ccc2aFF")}

                                                                                                        #Bracket
                                                                                                        overlay_slot:
                                                                                                            VerticalAlignment:=vertical_alignment.Fill
                                                                                                            HorizontalAlignment:=horizontal_alignment.Fill
                                                                                                            Padding:=margin{Left:=0.0, Top:=0.0, Right:=0.0, Bottom:=0.0}
                                                                                                            Widget:=stack_box:
                                                                                                                Orientation:=orientation.Horizontal
                                                                                                                #two instances of bracket texture to represent the range of rarities
                                                                                                                Slots:=array:

                                                                                                                    #stack_box_slot:
                                                                                                                        #VerticalAlignment:=vertical_alignment.Center
                                                                                                                        #HorizontalAlignment:=horizontal_alignment.Left
                                                                                                                        #Distribution:=option{1.0}
                                                                                                                        #Padding:=margin{Left:=0.0, Top:=0.0, Right:=0.0, Bottom:=0.0}
                                                                                                                        #Widget:=texture_block{DefaultImage:=Textures.General.T_RarityBracketLeft, DefaultDesiredSize:=vector2{X:=18.0, Y:=36.0}, DefaultTint:=MakeColorFromHex("#FFFFFF44")}

                                                                                                                    #stack_box_slot:
                                                                                                                        #VerticalAlignment:=vertical_alignment.Center
                                                                                                                        #HorizontalAlignment:=horizontal_alignment.Right
                                                                                                                        #Distribution:=option{1.0}
                                                                                                                        #Padding:=margin{Left:=0.0, Top:=0.0, Right:=0.0, Bottom:=0.0}
                                                                                                                        #Widget:=texture_block{DefaultImage:=Textures.General.T_RarityBracketRight, DefaultDesiredSize:=vector2{X:=18.0, Y:=36.0}, DefaultTint:=MakeColorFromHex("#FFFFFF44")}

                                                                                            stack_box_slot:
                                                                                                VerticalAlignment:=vertical_alignment.Center
                                                                                                HorizontalAlignment:=horizontal_alignment.Center
                                                                                                Distribution:=option{1.0}
                                                                                                Padding:=margin{Left:=0.0, Top:=0.0, Right:=0.0, Bottom:=0.0}
                                                                                                Widget:=overlay:
                                                                                                    Slots:=array:
                                                                                                        #Makeshift size box
                                                                                                        overlay_slot:
                                                                                                            VerticalAlignment:=vertical_alignment.Center
                                                                                                            HorizontalAlignment:=horizontal_alignment.Center
                                                                                                            Padding:=margin{Left:=0.0, Top:=0.0, Right:=0.0, Bottom:=0.0}
                                                                                                            Widget:=texture_block{DefaultImage:=Textures.General.T_GenSquTrans, DefaultDesiredSize:=vector2{X:=30.0, Y:=40.0}, DefaultTint:=MakeColorFromHex("#0000FF08")}

                                                                                                        #Overlay for rarity border and texture
                                                                                                        overlay_slot:
                                                                                                            VerticalAlignment:=vertical_alignment.Center
                                                                                                            HorizontalAlignment:=horizontal_alignment.Center
                                                                                                            Padding:=margin{Left:=0.0, Top:=0.0, Right:=0.0, Bottom:=0.0}
                                                                                                            Widget:=overlay:
                                                                                                                Slots:=array:
                                                                                                                                                                                                                
                                                                                                                    #Rarity texture border, when the rarity slot is empty
                                                                                                                    #overlay_slot:
                                                                                                                        #VerticalAlignment:=vertical_alignment.Center
                                                                                                                        #HorizontalAlignment:=horizontal_alignment.Center
                                                                                                                        #Padding:=margin{Left:=0.0, Top:=0.0, Right:=0.0, Bottom:=0.0}
                                                                                                                        #Widget:=texture_block{DefaultImage:=Textures.General.T_GenericBorderSquare2, DefaultDesiredSize:=vector2{X:=30.0, Y:=30.0}, DefaultTint:=MakeColorFromHex("#2e7be388")}
                                                                                                                                                                                                                            
                                                                                                                    #Rarity texture
                                                                                                                    overlay_slot:
                                                                                                                        VerticalAlignment:=vertical_alignment.Center
                                                                                                                        HorizontalAlignment:=horizontal_alignment.Center
                                                                                                                        Padding:=margin{Left:=0.0, Top:=0.0, Right:=0.0, Bottom:=0.0}
                                                                                                                        Widget:=texture_block{DefaultImage:=Textures.General.T_GenericSquare, DefaultDesiredSize:=vector2{X:=20.0, Y:=20.0}, DefaultTint:=MakeColorFromHex("#183379FF")}
                                                                                                                    
                                                                                                        #Bracket
                                                                                                        overlay_slot:
                                                                                                            VerticalAlignment:=vertical_alignment.Fill
                                                                                                            HorizontalAlignment:=horizontal_alignment.Fill
                                                                                                            Padding:=margin{Left:=0.0, Top:=0.0, Right:=0.0, Bottom:=0.0}
                                                                                                            Widget:=stack_box:
                                                                                                                Orientation:=orientation.Horizontal
                                                                                                                #two instances of bracket texture to represent the range of rarities
                                                                                                                Slots:=array:

                                                                                                                    stack_box_slot:
                                                                                                                        VerticalAlignment:=vertical_alignment.Center
                                                                                                                        HorizontalAlignment:=horizontal_alignment.Left
                                                                                                                        Distribution:=option{1.0}
                                                                                                                        Padding:=margin{Left:=0.0, Top:=0.0, Right:=0.0, Bottom:=0.0}
                                                                                                                        Widget:=texture_block{DefaultImage:=Textures.General.T_RarityBracketLeft, DefaultDesiredSize:=vector2{X:=18.0, Y:=36.0}, DefaultTint:=MakeColorFromHex("#a3a6aaff")}

                                                                                                                    #stack_box_slot:
                                                                                                                        #VerticalAlignment:=vertical_alignment.Center
                                                                                                                        #HorizontalAlignment:=horizontal_alignment.Right
                                                                                                                        #Distribution:=option{1.0}
                                                                                                                        #Padding:=margin{Left:=0.0, Top:=0.0, Right:=0.0, Bottom:=0.0}
                                                                                                                        #Widget:=texture_block{DefaultImage:=Textures.General.T_RarityBracketRight, DefaultDesiredSize:=vector2{X:=18.0, Y:=36.0}, DefaultTint:=MakeColorFromHex("#FFFFFF44")}

                                                                                            stack_box_slot:
                                                                                                VerticalAlignment:=vertical_alignment.Center
                                                                                                HorizontalAlignment:=horizontal_alignment.Center
                                                                                                Distribution:=option{1.0}
                                                                                                Padding:=margin{Left:=0.0, Top:=0.0, Right:=0.0, Bottom:=0.0}
                                                                                                Widget:=overlay:
                                                                                                    Slots:=array:
                                                                                                        #Makeshift size box
                                                                                                        overlay_slot:
                                                                                                            VerticalAlignment:=vertical_alignment.Center
                                                                                                            HorizontalAlignment:=horizontal_alignment.Center
                                                                                                            Padding:=margin{Left:=0.0, Top:=0.0, Right:=0.0, Bottom:=0.0}
                                                                                                            Widget:=texture_block{DefaultImage:=Textures.General.T_GenSquTrans, DefaultDesiredSize:=vector2{X:=30.0, Y:=40.0}, DefaultTint:=MakeColorFromHex("#0000FF08")}

                                                                                                        #Overlay for rarity border and texture
                                                                                                        overlay_slot:
                                                                                                            VerticalAlignment:=vertical_alignment.Center
                                                                                                            HorizontalAlignment:=horizontal_alignment.Center
                                                                                                            Padding:=margin{Left:=0.0, Top:=0.0, Right:=0.0, Bottom:=0.0}
                                                                                                            Widget:=overlay:
                                                                                                                Slots:=array:

                                                                                                                    #Rarity texture border, when the rarity slot is empty
                                                                                                                    #overlay_slot:
                                                                                                                        #VerticalAlignment:=vertical_alignment.Center
                                                                                                                        #HorizontalAlignment:=horizontal_alignment.Center
                                                                                                                        #Padding:=margin{Left:=0.0, Top:=0.0, Right:=0.0, Bottom:=0.0}
                                                                                                                        #Widget:=texture_block{DefaultImage:=Textures.General.T_GenericBorderSquare2, DefaultDesiredSize:=vector2{X:=30.0, Y:=30.0}, DefaultTint:=MakeColorFromHex("#2e7be388")}
                                                                                                                    
                                                                                                                    #Rarity texture
                                                                                                                    overlay_slot:
                                                                                                                        VerticalAlignment:=vertical_alignment.Center
                                                                                                                        HorizontalAlignment:=horizontal_alignment.Center
                                                                                                                        Padding:=margin{Left:=0.0, Top:=0.0, Right:=0.0, Bottom:=0.0}
                                                                                                                        Widget:=texture_block{DefaultImage:=Textures.General.T_GenericSquare, DefaultDesiredSize:=vector2{X:=20.0, Y:=20.0}, DefaultTint:=MakeColorFromHex("#5b1e79FF")}
                                                                                                                    
                                                                                                        #Bracket
                                                                                                        overlay_slot:
                                                                                                            VerticalAlignment:=vertical_alignment.Fill
                                                                                                            HorizontalAlignment:=horizontal_alignment.Fill
                                                                                                            Padding:=margin{Left:=0.0, Top:=0.0, Right:=0.0, Bottom:=0.0}
                                                                                                            Widget:=stack_box:
                                                                                                                Orientation:=orientation.Horizontal
                                                                                                                #two instances of bracket texture to represent the range of rarities
                                                                                                                Slots:=array:

                                                                                                                    #stack_box_slot:
                                                                                                                        #VerticalAlignment:=vertical_alignment.Center
                                                                                                                        #HorizontalAlignment:=horizontal_alignment.Left
                                                                                                                        #Distribution:=option{1.0}
                                                                                                                        #Padding:=margin{Left:=0.0, Top:=0.0, Right:=0.0, Bottom:=0.0}
                                                                                                                        #Widget:=texture_block{DefaultImage:=Textures.General.T_RarityBracketLeft, DefaultDesiredSize:=vector2{X:=18.0, Y:=36.0}, DefaultTint:=MakeColorFromHex("#FFFFFF44")}

                                                                                                                    #stack_box_slot:
                                                                                                                        #VerticalAlignment:=vertical_alignment.Center
                                                                                                                        #HorizontalAlignment:=horizontal_alignment.Right
                                                                                                                        #Distribution:=option{1.0}
                                                                                                                        #Padding:=margin{Left:=0.0, Top:=0.0, Right:=0.0, Bottom:=0.0}
                                                                                                                        #Widget:=texture_block{DefaultImage:=Textures.General.T_RarityBracketRight, DefaultDesiredSize:=vector2{X:=18.0, Y:=36.0}, DefaultTint:=MakeColorFromHex("#FFFFFF44")}

                                                                                            stack_box_slot:
                                                                                                VerticalAlignment:=vertical_alignment.Center
                                                                                                HorizontalAlignment:=horizontal_alignment.Center
                                                                                                Distribution:=option{1.0}
                                                                                                Padding:=margin{Left:=0.0, Top:=0.0, Right:=0.0, Bottom:=0.0}
                                                                                                Widget:=overlay:
                                                                                                    Slots:=array:
                                                                                                        #Makeshift size box
                                                                                                        overlay_slot:
                                                                                                            VerticalAlignment:=vertical_alignment.Center
                                                                                                            HorizontalAlignment:=horizontal_alignment.Center
                                                                                                            Padding:=margin{Left:=0.0, Top:=0.0, Right:=0.0, Bottom:=0.0}
                                                                                                            Widget:=texture_block{DefaultImage:=Textures.General.T_GenSquTrans, DefaultDesiredSize:=vector2{X:=30.0, Y:=40.0}, DefaultTint:=MakeColorFromHex("#0000FF08")}

                                                                                                        #Overlay for rarity border and texture
                                                                                                        overlay_slot:
                                                                                                            VerticalAlignment:=vertical_alignment.Center
                                                                                                            HorizontalAlignment:=horizontal_alignment.Center
                                                                                                            Padding:=margin{Left:=0.0, Top:=0.0, Right:=0.0, Bottom:=0.0}
                                                                                                            Widget:=overlay:
                                                                                                                Slots:=array:

                                                                                                                    #Rarity texture border, when the rarity slot is empty
                                                                                                                    #overlay_slot:
                                                                                                                        #VerticalAlignment:=vertical_alignment.Center
                                                                                                                        #HorizontalAlignment:=horizontal_alignment.Center
                                                                                                                        #Padding:=margin{Left:=0.0, Top:=0.0, Right:=0.0, Bottom:=0.0}
                                                                                                                        #Widget:=texture_block{DefaultImage:=Textures.General.T_GenericBorderSquare2, DefaultDesiredSize:=vector2{X:=30.0, Y:=30.0}, DefaultTint:=MakeColorFromHex("#2e7be388")}
                                                                                                                    
                                                                                                                    #Rarity texture
                                                                                                                    overlay_slot:
                                                                                                                        VerticalAlignment:=vertical_alignment.Center
                                                                                                                        HorizontalAlignment:=horizontal_alignment.Center
                                                                                                                        Padding:=margin{Left:=0.0, Top:=0.0, Right:=0.0, Bottom:=0.0}
                                                                                                                        Widget:=texture_block{DefaultImage:=Textures.General.T_GenericSquare, DefaultDesiredSize:=vector2{X:=20.0, Y:=20.0}, DefaultTint:=MakeColorFromHex("#674f28FF")}

                                                                                                        #Bracket
                                                                                                        overlay_slot:
                                                                                                            VerticalAlignment:=vertical_alignment.Fill
                                                                                                            HorizontalAlignment:=horizontal_alignment.Fill
                                                                                                            Padding:=margin{Left:=0.0, Top:=0.0, Right:=0.0, Bottom:=0.0}
                                                                                                            Widget:=stack_box:
                                                                                                                Orientation:=orientation.Horizontal
                                                                                                                #two instances of bracket texture to represent the range of rarities
                                                                                                                Slots:=array:

                                                                                                                    #stack_box_slot:
                                                                                                                        #VerticalAlignment:=vertical_alignment.Center
                                                                                                                        #HorizontalAlignment:=horizontal_alignment.Left
                                                                                                                        #Distribution:=option{1.0}
                                                                                                                        #Padding:=margin{Left:=0.0, Top:=0.0, Right:=0.0, Bottom:=0.0}
                                                                                                                        #Widget:=texture_block{DefaultImage:=Textures.General.T_RarityBracketLeft, DefaultDesiredSize:=vector2{X:=18.0, Y:=36.0}, DefaultTint:=MakeColorFromHex("#FFFFFF44")}

                                                                                                                    #stack_box_slot:
                                                                                                                        #VerticalAlignment:=vertical_alignment.Center
                                                                                                                        #HorizontalAlignment:=horizontal_alignment.Right
                                                                                                                        #Distribution:=option{1.0}
                                                                                                                        #Padding:=margin{Left:=0.0, Top:=0.0, Right:=0.0, Bottom:=0.0}
                                                                                                                        #Widget:=texture_block{DefaultImage:=Textures.General.T_RarityBracketRight, DefaultDesiredSize:=vector2{X:=18.0, Y:=36.0}, DefaultTint:=MakeColorFromHex("#FFFFFF44")}

                                                                                            stack_box_slot:
                                                                                                VerticalAlignment:=vertical_alignment.Center
                                                                                                HorizontalAlignment:=horizontal_alignment.Center
                                                                                                Distribution:=option{1.0}
                                                                                                Padding:=margin{Left:=0.0, Top:=0.0, Right:=0.0, Bottom:=0.0}
                                                                                                Widget:=overlay:
                                                                                                    Slots:=array:
                                                                                                        #Makeshift size box
                                                                                                        overlay_slot:
                                                                                                            VerticalAlignment:=vertical_alignment.Center
                                                                                                            HorizontalAlignment:=horizontal_alignment.Center
                                                                                                            Padding:=margin{Left:=0.0, Top:=0.0, Right:=0.0, Bottom:=0.0}
                                                                                                            Widget:=texture_block{DefaultImage:=Textures.General.T_GenSquTrans, DefaultDesiredSize:=vector2{X:=30.0, Y:=40.0}, DefaultTint:=MakeColorFromHex("#0000FF08")}

                                                                                                        #Overlay for rarity border and texture
                                                                                                        overlay_slot:
                                                                                                            VerticalAlignment:=vertical_alignment.Center
                                                                                                            HorizontalAlignment:=horizontal_alignment.Center
                                                                                                            Padding:=margin{Left:=0.0, Top:=0.0, Right:=0.0, Bottom:=0.0}
                                                                                                            Widget:=overlay:
                                                                                                                Slots:=array:
                                                                                                        
                                                                                                                    #Rarity texture border, when the rarity slot is empty
                                                                                                                    #overlay_slot:
                                                                                                                        #VerticalAlignment:=vertical_alignment.Center
                                                                                                                        #HorizontalAlignment:=horizontal_alignment.Center
                                                                                                                        #Padding:=margin{Left:=0.0, Top:=0.0, Right:=0.0, Bottom:=0.0}
                                                                                                                        #Widget:=texture_block{DefaultImage:=Textures.General.T_GenericBorderSquare2, DefaultDesiredSize:=vector2{X:=30.0, Y:=30.0}, DefaultTint:=MakeColorFromHex("#2e7be388")}
                                                                                                                    
                                                                                                                    #Rarity texture
                                                                                                                    overlay_slot:
                                                                                                                        VerticalAlignment:=vertical_alignment.Center
                                                                                                                        HorizontalAlignment:=horizontal_alignment.Center
                                                                                                                        Padding:=margin{Left:=0.0, Top:=0.0, Right:=0.0, Bottom:=0.0}
                                                                                                                        Widget:=texture_block{DefaultImage:=Textures.General.T_GenericSquare, DefaultDesiredSize:=vector2{X:=20.0, Y:=20.0}, DefaultTint:=MakeColorFromHex("#78711fFF")}
                                                                                                                    
                                                                                                        #Bracket
                                                                                                        overlay_slot:
                                                                                                            VerticalAlignment:=vertical_alignment.Fill
                                                                                                            HorizontalAlignment:=horizontal_alignment.Fill
                                                                                                            Padding:=margin{Left:=0.0, Top:=0.0, Right:=0.0, Bottom:=0.0}
                                                                                                            Widget:=stack_box:
                                                                                                                Orientation:=orientation.Horizontal
                                                                                                                #two instances of bracket texture to represent the range of rarities
                                                                                                                Slots:=array:

                                                                                                                    #stack_box_slot:
                                                                                                                        #VerticalAlignment:=vertical_alignment.Center
                                                                                                                        #HorizontalAlignment:=horizontal_alignment.Left
                                                                                                                        #Distribution:=option{1.0}
                                                                                                                        #Padding:=margin{Left:=0.0, Top:=0.0, Right:=0.0, Bottom:=0.0}
                                                                                                                        #Widget:=texture_block{DefaultImage:=Textures.General.T_RarityBracketLeft, DefaultDesiredSize:=vector2{X:=18.0, Y:=36.0}, DefaultTint:=MakeColorFromHex("#FFFFFF44")}

                                                                                                                    stack_box_slot:
                                                                                                                        VerticalAlignment:=vertical_alignment.Center
                                                                                                                        HorizontalAlignment:=horizontal_alignment.Right
                                                                                                                        Distribution:=option{1.0}
                                                                                                                        Padding:=margin{Left:=0.0, Top:=0.0, Right:=0.0, Bottom:=0.0}
                                                                                                                        Widget:=texture_block{DefaultImage:=Textures.General.T_RarityBracketRight, DefaultDesiredSize:=vector2{X:=18.0, Y:=36.0}, DefaultTint:=MakeColorFromHex("#a3a6aaff")}

                    #overlay_slot:
                        #VerticalAlignment:=vertical_alignment.Fill
                        #HorizontalAlignment:=horizontal_alignment.Fill
                        #Padding:=margin{Left:=0.0, Top:=0.0, Right:=0.0, Bottom:=0.0}
                        #Widget:=texture_block{DefaultImage:=Textures.General.T_LevelPointer, DefaultDesiredSize:=vector2{X:=20.0, Y:=20.0}, DefaultTint:=MakeColorFromHex("#7de2FFFF")}

            #Definition for the Seventh level row
            LevelRow7:=overlay:
                Slots:=array:
                    overlay_slot:
                        #Makeshift size box
                        VerticalAlignment:=vertical_alignment.Center
                        HorizontalAlignment:=horizontal_alignment.Center
                        Padding:=margin{Left:=0.0, Top:=0.0, Right:=0.0, Bottom:=0.0}
                        Widget:=texture_block{DefaultImage:=Textures.General.T_GenSquTrans, DefaultDesiredSize:=vector2{X:=1250.0, Y:=114.0}, DefaultTint:=MakeColorFromHex("#0000FFFF")}

                    overlay_slot:
                        #Rarity level content
                        VerticalAlignment:=vertical_alignment.Center
                        HorizontalAlignment:=horizontal_alignment.Left
                        Padding:=margin{Left:=0.0, Top:=0.0, Right:=0.0, Bottom:=0.0}
                        Widget:=overlay:
                            Slots:=array:

                                overlay_slot:
                                    #Makeshift size box
                                    VerticalAlignment:=vertical_alignment.Center
                                    HorizontalAlignment:=horizontal_alignment.Center
                                    Padding:=margin{Left:=0.0, Top:=0.0, Right:=0.0, Bottom:=0.0}
                                    Widget:=texture_block{DefaultImage:=Textures.General.T_GenericSquare, DefaultDesiredSize:=vector2{X:=1190.0, Y:=114.0}, DefaultTint:=MakeColorFromHex("#010a1832")}

                                overlay_slot:
                                    #Highlight texture, over the content but under the rarity flag
                                    VerticalAlignment:=vertical_alignment.Center
                                    HorizontalAlignment:=horizontal_alignment.Center
                                    Padding:=margin{Left:=0.0, Top:=0.0, Right:=0.0, Bottom:=0.0}
                                    Widget:=texture_block{DefaultImage:=Textures.General.T_LevelBorder, DefaultDesiredSize:=vector2{X:=1190.0, Y:=114.0}, DefaultTint:=MakeColorFromHex("#174583FF")}

                                    #Horizontal stack box
                                overlay_slot:
                                    VerticalAlignment:=vertical_alignment.Fill
                                    HorizontalAlignment:=horizontal_alignment.Left
                                    Padding:=margin{Left:=0.0, Top:=0.0, Right:=0.0, Bottom:=0.0}
                                    Widget:=stack_box:
                                        Orientation:=orientation.Horizontal
                                        Slots:=array:
                                            #Level flag
                                            stack_box_slot:
                                                VerticalAlignment:=vertical_alignment.Top
                                                HorizontalAlignment:=horizontal_alignment.Left
                                                Padding:=margin{Left:=0.0, Top:=0.0, Right:=15.0, Bottom:=0.0}
                                                Widget:=overlay:

                                                    Slots:=array:
                                                        overlay_slot:
                                                            VerticalAlignment:=vertical_alignment.Center
                                                            HorizontalAlignment:=horizontal_alignment.Center
                                                            Padding:=margin{Left:=0.0, Top:=0.0, Right:=0.0, Bottom:=0.0}
                                                            #Tier level makeshift size box
                                                            Widget:=texture_block{DefaultImage:=Textures.General.T_GenSquTrans, DefaultDesiredSize:=vector2{X:=150.0, Y:=114.0}, DefaultTint:=MakeColorFromHex("#FF0000FF")}

                                                        overlay_slot:
                                                            VerticalAlignment:=vertical_alignment.Top
                                                            HorizontalAlignment:=horizontal_alignment.Left
                                                            Padding:=margin{Left:=0.0, Top:=0.0, Right:=0.0, Bottom:=0.0}
                                                            #Selected level yellow flag
                                                            Widget:=overlay:

                                                                Slots:=array:
                                                                    overlay_slot:
                                                                        VerticalAlignment:=vertical_alignment.Center
                                                                        HorizontalAlignment:=horizontal_alignment.Center
                                                                        Padding:=margin{Left:=0.0, Top:=0.0, Right:=0.0, Bottom:=0.0}
                                                                        #level flag texture
                                                                        Widget:=texture_block{DefaultImage:=Textures.General.T_GenLevelFlag, DefaultDesiredSize:=vector2{X:=140.0, Y:=45.0}, DefaultTint:=MakeColorFromHex("#f7ff1aFF")}

                                                                    overlay_slot:
                                                                        VerticalAlignment:=vertical_alignment.Center
                                                                        HorizontalAlignment:=horizontal_alignment.Center
                                                                        Padding:=margin{Left:=0.0, Top:=0.0, Right:=0.0, Bottom:=0.0}
                                                                        #level flag text
                                                                        Widget:=text_block{DefaultText := StringToMessage("Level 0"), DefaultTextColor:=MakeColorFromHex("#000000FF")}

                                                        overlay_slot:
                                                            VerticalAlignment:=vertical_alignment.Fill
                                                            HorizontalAlignment:=horizontal_alignment.Center
                                                            Padding:=margin{Left:=0.0, Top:=5.0, Right:=0.0, Bottom:=5.0}
                                                            #Unselected level and lock
                                                            Widget:=overlay:
                                                                Slots:=array:
                                                                    
                                                                    overlay_slot:
                                                                        VerticalAlignment:=vertical_alignment.Top
                                                                        HorizontalAlignment:=horizontal_alignment.Center
                                                                        Padding:=margin{Left:=0.0, Top:=0.0, Right:=0.0, Bottom:=0.0}
                                                                        #level flag texture
                                                                        Widget:=text_block{DefaultText := StringToMessage("Level 0"), DefaultTextColor:=MakeColorFromHex("#a3a6aaff")}

                                                                    overlay_slot:
                                                                        VerticalAlignment:=vertical_alignment.Bottom
                                                                        HorizontalAlignment:=horizontal_alignment.Center
                                                                        Padding:=margin{Left:=0.0, Top:=0.0, Right:=0.0, Bottom:=0.0}
                                                                        #level flag texture
                                                                        Widget:=texture_block{DefaultImage:=Textures.General.T_Lock, DefaultDesiredSize:=vector2{X:=52.0, Y:=58.0}, DefaultTint:=MakeColorFromHex("#a3a6aaff")}
                                            
                                            #Level weapon roster
                                            stack_box_slot:
                                                VerticalAlignment:=vertical_alignment.Center
                                                HorizontalAlignment:=horizontal_alignment.Center
                                                Padding:=margin{Left:=0.0, Top:=0.0, Right:=8.0, Bottom:=0.0}
                                                Widget:=overlay:
                                                    Slots:=array:
                                                        #Weapon Icon - Pistols
                                                        overlay_slot:
                                                            VerticalAlignment:=vertical_alignment.Center
                                                            HorizontalAlignment:=horizontal_alignment.Center
                                                            Padding:=margin{Left:=0.0, Top:=0.0, Right:=0.0, Bottom:=0.0}
                                                            #Weapon icon BG
                                                            Widget:=texture_block{DefaultImage:=Textures.General.T_GenericSquare, DefaultDesiredSize:=vector2{X:=78.0, Y:=78.0}, DefaultTint:=MakeColorFromHex("#000c1d88")}

                                                        overlay_slot:
                                                            VerticalAlignment:=vertical_alignment.Center
                                                            HorizontalAlignment:=horizontal_alignment.Center
                                                            Padding:=margin{Left:=0.0, Top:=0.0, Right:=0.0, Bottom:=0.0}
                                                            #Weapon icon Border
                                                            Widget:=texture_block{DefaultImage:=Textures.General.T_GenericBorderSquare, DefaultDesiredSize:=vector2{X:=78.0, Y:=78.0}, DefaultTint:=MakeColorFromHex("#174583FF")}

                                                        overlay_slot:
                                                            VerticalAlignment:=vertical_alignment.Top
                                                            HorizontalAlignment:=horizontal_alignment.Center
                                                            Padding:=margin{Left:=0.0, Top:=5.0, Right:=0.0, Bottom:=0.0}
                                                            #Weapon icon - pistol
                                                            Widget:=texture_block{DefaultImage:=Textures.Items.T_IconPistol, DefaultDesiredSize:=vector2{X:=50.0, Y:=50.0}, DefaultTint:=MakeColorFromHex("#a3a6aaff")}

                                                        overlay_slot:
                                                            VerticalAlignment:=vertical_alignment.Bottom
                                                            HorizontalAlignment:=horizontal_alignment.Center
                                                            Padding:=margin{Left:=0.0, Top:=0.0, Right:=0.0, Bottom:=5.0}
                                                            #Weapon header - pistol
                                                            Widget:=texture_block{DefaultImage:=Textures.Items.T_HeaderPistol, DefaultDesiredSize:=vector2{X:=78.0, Y:=18.0}, DefaultTint:=MakeColorFromHex("#a3a6aaff")}
                                                        
                                            stack_box_slot:
                                                VerticalAlignment:=vertical_alignment.Center
                                                HorizontalAlignment:=horizontal_alignment.Center
                                                Padding:=margin{Left:=0.0, Top:=0.0, Right:=8.0, Bottom:=0.0}
                                                Widget:=overlay:
                                                    Slots:=array:
                                                        #Weapon Icon - Shotgun
                                                        overlay_slot:
                                                            VerticalAlignment:=vertical_alignment.Center
                                                            HorizontalAlignment:=horizontal_alignment.Center
                                                            Padding:=margin{Left:=0.0, Top:=0.0, Right:=0.0, Bottom:=0.0}
                                                            #Weapon icon BG
                                                            Widget:=texture_block{DefaultImage:=Textures.General.T_GenericSquare, DefaultDesiredSize:=vector2{X:=78.0, Y:=78.0}, DefaultTint:=MakeColorFromHex("#000c1d88")}

                                                        overlay_slot:
                                                            VerticalAlignment:=vertical_alignment.Center
                                                            HorizontalAlignment:=horizontal_alignment.Center
                                                            Padding:=margin{Left:=0.0, Top:=0.0, Right:=0.0, Bottom:=0.0}
                                                            #Weapon icon Border
                                                            Widget:=texture_block{DefaultImage:=Textures.General.T_GenericBorderSquare, DefaultDesiredSize:=vector2{X:=78.0, Y:=78.0}, DefaultTint:=MakeColorFromHex("#174583FF")}

                                                        overlay_slot:
                                                            VerticalAlignment:=vertical_alignment.Top
                                                            HorizontalAlignment:=horizontal_alignment.Center
                                                            Padding:=margin{Left:=0.0, Top:=5.0, Right:=0.0, Bottom:=0.0}
                                                            #Weapon icon - shotgun
                                                            Widget:=texture_block{DefaultImage:=Textures.Items.T_IconShotgun, DefaultDesiredSize:=vector2{X:=50.0, Y:=50.0}, DefaultTint:=MakeColorFromHex("#a3a6aaff")}

                                                        overlay_slot:
                                                            VerticalAlignment:=vertical_alignment.Bottom
                                                            HorizontalAlignment:=horizontal_alignment.Center
                                                            Padding:=margin{Left:=0.0, Top:=0.0, Right:=0.0, Bottom:=5.0}
                                                            #Weapon header - shotgun
                                                            Widget:=texture_block{DefaultImage:=Textures.Items.T_HeaderShotgun, DefaultDesiredSize:=vector2{X:=78.0, Y:=18.0}, DefaultTint:=MakeColorFromHex("#a3a6aaff")}

                                            stack_box_slot:
                                                VerticalAlignment:=vertical_alignment.Center
                                                HorizontalAlignment:=horizontal_alignment.Center
                                                Padding:=margin{Left:=0.0, Top:=0.0, Right:=8.0, Bottom:=0.0}
                                                Widget:=overlay:
                                                    Slots:=array:
                                                        #Weapon Icon - SMG
                                                        overlay_slot:
                                                            VerticalAlignment:=vertical_alignment.Top
                                                            HorizontalAlignment:=horizontal_alignment.Center
                                                            Padding:=margin{Left:=0.0, Top:=0.0, Right:=0.0, Bottom:=0.0}
                                                            #Weapon icon BG
                                                            Widget:=texture_block{DefaultImage:=Textures.General.T_GenericSquare, DefaultDesiredSize:=vector2{X:=78.0, Y:=78.0}, DefaultTint:=MakeColorFromHex("#000c1d88")}

                                                        overlay_slot:
                                                            VerticalAlignment:=vertical_alignment.Center
                                                            HorizontalAlignment:=horizontal_alignment.Center
                                                            Padding:=margin{Left:=0.0, Top:=0.0, Right:=0.0, Bottom:=0.0}
                                                            #Weapon icon Border
                                                            Widget:=texture_block{DefaultImage:=Textures.General.T_GenericBorderSquare, DefaultDesiredSize:=vector2{X:=78.0, Y:=78.0}, DefaultTint:=MakeColorFromHex("#174583FF")}

                                                        overlay_slot:
                                                            VerticalAlignment:=vertical_alignment.Top
                                                            HorizontalAlignment:=horizontal_alignment.Center
                                                            Padding:=margin{Left:=0.0, Top:=5.0, Right:=0.0, Bottom:=0.0}
                                                            #Weapon icon - SMG
                                                            Widget:=texture_block{DefaultImage:=Textures.Items.T_IconSMG, DefaultDesiredSize:=vector2{X:=50.0, Y:=50.0}, DefaultTint:=MakeColorFromHex("#a3a6aaff")}

                                                        overlay_slot:
                                                            VerticalAlignment:=vertical_alignment.Bottom
                                                            HorizontalAlignment:=horizontal_alignment.Center
                                                            Padding:=margin{Left:=0.0, Top:=0.0, Right:=0.0, Bottom:=5.0}
                                                            #Weapon header - SMG
                                                            Widget:=texture_block{DefaultImage:=Textures.Items.T_HeaderSMGs, DefaultDesiredSize:=vector2{X:=78.0, Y:=18.0}, DefaultTint:=MakeColorFromHex("#a3a6aaff")}
                                                            

                                            stack_box_slot:
                                                VerticalAlignment:=vertical_alignment.Center
                                                HorizontalAlignment:=horizontal_alignment.Center
                                                Padding:=margin{Left:=0.0, Top:=0.0, Right:=8.0, Bottom:=0.0}
                                                Widget:=overlay:
                                                    Slots:=array:
                                                        #Weapon Icon - Rifle
                                                        overlay_slot:
                                                            VerticalAlignment:=vertical_alignment.Center
                                                            HorizontalAlignment:=horizontal_alignment.Center
                                                            Padding:=margin{Left:=0.0, Top:=0.0, Right:=0.0, Bottom:=0.0}
                                                            #Weapon icon BG
                                                            Widget:=texture_block{DefaultImage:=Textures.General.T_GenericSquare, DefaultDesiredSize:=vector2{X:=78.0, Y:=78.0}, DefaultTint:=MakeColorFromHex("#000c1d88")}

                                                        overlay_slot:
                                                            VerticalAlignment:=vertical_alignment.Top
                                                            HorizontalAlignment:=horizontal_alignment.Center
                                                            Padding:=margin{Left:=0.0, Top:=0.0, Right:=0.0, Bottom:=0.0}
                                                            #Weapon icon Border
                                                            Widget:=texture_block{DefaultImage:=Textures.General.T_GenericBorderSquare, DefaultDesiredSize:=vector2{X:=78.0, Y:=78.0}, DefaultTint:=MakeColorFromHex("#174583FF")}

                                                        overlay_slot:
                                                            VerticalAlignment:=vertical_alignment.Top
                                                            HorizontalAlignment:=horizontal_alignment.Center
                                                            Padding:=margin{Left:=0.0, Top:=5.0, Right:=0.0, Bottom:=0.0}
                                                            #Weapon icon - Rifle
                                                            Widget:=texture_block{DefaultImage:=Textures.Items.T_IconRifle, DefaultDesiredSize:=vector2{X:=50.0, Y:=50.0}, DefaultTint:=MakeColorFromHex("#a3a6aaff")}

                                                        overlay_slot:
                                                            VerticalAlignment:=vertical_alignment.Bottom
                                                            HorizontalAlignment:=horizontal_alignment.Center
                                                            Padding:=margin{Left:=0.0, Top:=0.0, Right:=0.0, Bottom:=5.0}
                                                            #Weapon header - Rifle
                                                            Widget:=texture_block{DefaultImage:=Textures.Items.T_HeaderRifle, DefaultDesiredSize:=vector2{X:=78.0, Y:=18.0}, DefaultTint:=MakeColorFromHex("#a3a6aaff")}

                                            stack_box_slot:
                                                VerticalAlignment:=vertical_alignment.Center
                                                HorizontalAlignment:=horizontal_alignment.Center
                                                Padding:=margin{Left:=0.0, Top:=0.0, Right:=8.0, Bottom:=0.0}
                                                Widget:=overlay:
                                                    Slots:=array:
                                                        #Weapon Icon - Explosive
                                                        overlay_slot:
                                                            VerticalAlignment:=vertical_alignment.Center
                                                            HorizontalAlignment:=horizontal_alignment.Center
                                                            Padding:=margin{Left:=0.0, Top:=0.0, Right:=0.0, Bottom:=0.0}
                                                            #Weapon icon BG
                                                            Widget:=texture_block{DefaultImage:=Textures.General.T_GenericSquare, DefaultDesiredSize:=vector2{X:=78.0, Y:=78.0}, DefaultTint:=MakeColorFromHex("#000c1d88")}

                                                        overlay_slot:
                                                            VerticalAlignment:=vertical_alignment.Center
                                                            HorizontalAlignment:=horizontal_alignment.Center
                                                            Padding:=margin{Left:=0.0, Top:=0.0, Right:=0.0, Bottom:=0.0}
                                                            #Weapon icon Border
                                                            Widget:=texture_block{DefaultImage:=Textures.General.T_GenericBorderSquare, DefaultDesiredSize:=vector2{X:=78.0, Y:=78.0}, DefaultTint:=MakeColorFromHex("#174583FF")}

                                                        overlay_slot:
                                                            VerticalAlignment:=vertical_alignment.Top
                                                            HorizontalAlignment:=horizontal_alignment.Center
                                                            Padding:=margin{Left:=0.0, Top:=5.0, Right:=0.0, Bottom:=0.0}
                                                            #Weapon icon - Explosive
                                                            Widget:=texture_block{DefaultImage:=Textures.Items.T_IconExplosive, DefaultDesiredSize:=vector2{X:=50.0, Y:=50.0}, DefaultTint:=MakeColorFromHex("#a3a6aaff")}

                                                        overlay_slot:
                                                            VerticalAlignment:=vertical_alignment.Bottom
                                                            HorizontalAlignment:=horizontal_alignment.Center
                                                            Padding:=margin{Left:=0.0, Top:=0.0, Right:=0.0, Bottom:=5.0}
                                                            #Weapon header - Explosive
                                                            Widget:=texture_block{DefaultImage:=Textures.Items.T_HeaderExplosive, DefaultDesiredSize:=vector2{X:=78.0, Y:=18.0}, DefaultTint:=MakeColorFromHex("#a3a6aaff")}

                                            stack_box_slot:
                                                VerticalAlignment:=vertical_alignment.Center
                                                HorizontalAlignment:=horizontal_alignment.Center
                                                Padding:=margin{Left:=0.0, Top:=0.0, Right:=8.0, Bottom:=0.0}
                                                Widget:=overlay:
                                                    Slots:=array:
                                                        #Weapon Icon - Special
                                                        overlay_slot:
                                                            VerticalAlignment:=vertical_alignment.Center
                                                            HorizontalAlignment:=horizontal_alignment.Center
                                                            Padding:=margin{Left:=0.0, Top:=0.0, Right:=0.0, Bottom:=0.0}
                                                            #Weapon icon BG
                                                            Widget:=texture_block{DefaultImage:=Textures.General.T_GenericSquare, DefaultDesiredSize:=vector2{X:=78.0, Y:=78.0}, DefaultTint:=MakeColorFromHex("#000c1d88")}

                                                        overlay_slot:
                                                            VerticalAlignment:=vertical_alignment.Center
                                                            HorizontalAlignment:=horizontal_alignment.Center
                                                            Padding:=margin{Left:=0.0, Top:=0.0, Right:=0.0, Bottom:=0.0}
                                                            #Weapon icon Border
                                                            Widget:=texture_block{DefaultImage:=Textures.General.T_GenericBorderSquare, DefaultDesiredSize:=vector2{X:=78.0, Y:=78.0}, DefaultTint:=MakeColorFromHex("#174583FF")}

                                                        overlay_slot:
                                                            VerticalAlignment:=vertical_alignment.Top
                                                            HorizontalAlignment:=horizontal_alignment.Center
                                                            Padding:=margin{Left:=0.0, Top:=5.0, Right:=0.0, Bottom:=0.0}
                                                            #Weapon icon - special
                                                            Widget:=texture_block{DefaultImage:=Textures.Items.T_IconArtifact, DefaultDesiredSize:=vector2{X:=50.0, Y:=50.0}, DefaultTint:=MakeColorFromHex("#a3a6aaff")}

                                                        overlay_slot:
                                                            VerticalAlignment:=vertical_alignment.Bottom
                                                            HorizontalAlignment:=horizontal_alignment.Center
                                                            Padding:=margin{Left:=0.0, Top:=0.0, Right:=0.0, Bottom:=5.0}
                                                            #Weapon header - Special
                                                            Widget:=texture_block{DefaultImage:=Textures.Items.T_HeaderSpecial, DefaultDesiredSize:=vector2{X:=78.0, Y:=18.0}, DefaultTint:=MakeColorFromHex("#a3a6aaff")}

                                            stack_box_slot:
                                                VerticalAlignment:=vertical_alignment.Center
                                                HorizontalAlignment:=horizontal_alignment.Center
                                                Padding:=margin{Left:=0.0, Top:=0.0, Right:=8.0, Bottom:=0.0}
                                                Widget:=overlay:
                                                    Slots:=array:
                                                        #Weapon Icon - DMR Sniper
                                                        overlay_slot:
                                                            VerticalAlignment:=vertical_alignment.Center
                                                            HorizontalAlignment:=horizontal_alignment.Center
                                                            Padding:=margin{Left:=0.0, Top:=0.0, Right:=0.0, Bottom:=0.0}
                                                            #Weapon icon BG
                                                            Widget:=texture_block{DefaultImage:=Textures.General.T_GenericSquare, DefaultDesiredSize:=vector2{X:=78.0, Y:=78.0}, DefaultTint:=MakeColorFromHex("#000c1d88")}

                                                        overlay_slot:
                                                            VerticalAlignment:=vertical_alignment.Center
                                                            HorizontalAlignment:=horizontal_alignment.Center
                                                            Padding:=margin{Left:=0.0, Top:=0.0, Right:=0.0, Bottom:=0.0}
                                                            #Weapon icon Border
                                                            Widget:=texture_block{DefaultImage:=Textures.General.T_GenericBorderSquare, DefaultDesiredSize:=vector2{X:=78.0, Y:=78.0}, DefaultTint:=MakeColorFromHex("#174583FF")}

                                                        overlay_slot:
                                                            VerticalAlignment:=vertical_alignment.Top
                                                            HorizontalAlignment:=horizontal_alignment.Center
                                                            Padding:=margin{Left:=0.0, Top:=5.0, Right:=0.0, Bottom:=0.0}
                                                            #Weapon icon - Sniper
                                                            Widget:=texture_block{DefaultImage:=Textures.Items.T_IconSniper, DefaultDesiredSize:=vector2{X:=50.0, Y:=50.0}, DefaultTint:=MakeColorFromHex("#a3a6aaff")}

                                                        overlay_slot:
                                                            VerticalAlignment:=vertical_alignment.Bottom
                                                            HorizontalAlignment:=horizontal_alignment.Center
                                                            Padding:=margin{Left:=0.0, Top:=0.0, Right:=0.0, Bottom:=5.0}
                                                            #Weapon header - DMR Sniper
                                                            Widget:=texture_block{DefaultImage:=Textures.Items.T_HeaderDMRSniper, DefaultDesiredSize:=vector2{X:=78.0, Y:=18.0}, DefaultTint:=MakeColorFromHex("#a3a6aaff")}

                                            stack_box_slot:
                                                VerticalAlignment:=vertical_alignment.Fill
                                                HorizontalAlignment:=horizontal_alignment.Center #Center
                                                Padding:=margin{Left:=0.0, Top:=0.0, Right:=0.0, Bottom:=0.0}
                                                Widget:=overlay:
                                                    Slots:=array:
                                                        overlay_slot:
                                                            #Makeshift size box
                                                            VerticalAlignment:=vertical_alignment.Center
                                                            HorizontalAlignment:=horizontal_alignment.Center
                                                            Padding:=margin{Left:=0.0, Top:=0.0, Right:=0.0, Bottom:=0.0}
                                                            Widget:=texture_block{DefaultImage:=Textures.General.T_GenSquTrans, DefaultDesiredSize:=vector2{X:=418.0, Y:=100.0}, DefaultTint:=MakeColorFromHex("#00FF0011")}

                                                        overlay_slot:
                                                            #Vertical stack box to add all the rarity information
                                                            VerticalAlignment:=vertical_alignment.Center
                                                            HorizontalAlignment:=horizontal_alignment.Center
                                                            Padding:=margin{Left:=0.0, Top:=0.0, Right:=0.0, Bottom:=0.0}
                                                            Widget:=stack_box:
                                                                Orientation:=orientation.Vertical
                                                                Slots:=array:
                                                                    stack_box_slot:
                                                                        #Horizontal Stack box for max rarity available
                                                                        VerticalAlignment:=vertical_alignment.Top
                                                                        HorizontalAlignment:=horizontal_alignment.Center
                                                                        Padding:=margin{Left:=0.0, Top:=0.0, Right:=0.0, Bottom:=0.0}
                                                                        Widget:=stack_box:
                                                                            Orientation:=orientation.Horizontal
                                                                            Slots:=array:
                                                                                stack_box_slot:
                                                                                    VerticalAlignment:=vertical_alignment.Center
                                                                                    HorizontalAlignment:=horizontal_alignment.Center
                                                                                    Padding:=margin{Left:=0.0, Top:=0.0, Right:=0.0, Bottom:=0.0}
                                                                                    Widget:=text_block{DefaultText := StringToMessage("Rarity chance up to "), DefaultTextColor:=MakeColorFromHex("#a3a6aaff")}

                                                                                stack_box_slot:
                                                                                    VerticalAlignment:=vertical_alignment.Center
                                                                                    HorizontalAlignment:=horizontal_alignment.Center
                                                                                    Padding:=margin{Left:=0.0, Top:=0.0, Right:=0.0, Bottom:=0.0}
                                                                                    Widget:=texture_block{DefaultImage:=Textures.Rarity.T_FlagCommon, DefaultDesiredSize:=vector2{X:=104.0, Y:=36.0}, DefaultTint:=MakeColorFromHex("#a3a6aaff")}

                                                                    stack_box_slot:
                                                                        #Horizontal overlay for overall rarity scheme
                                                                        VerticalAlignment:=vertical_alignment.Bottom
                                                                        HorizontalAlignment:=horizontal_alignment.Center
                                                                        Padding:=margin{Left:=0.0, Top:=0.0, Right:=0.0, Bottom:=0.0}
                                                                        Widget:=overlay:
                                                                            Slots:=array:
                                                                                
                                                                                #Makeshift size box for the rarity scheme
                                                                                overlay_slot:
                                                                                    VerticalAlignment:=vertical_alignment.Center
                                                                                    HorizontalAlignment:=horizontal_alignment.Center
                                                                                    Padding:=margin{Left:=0.0, Top:=0.0, Right:=0.0, Bottom:=0.0}
                                                                                    Widget:=texture_block{DefaultImage:=Textures.General.T_GenSquTrans, DefaultDesiredSize:=vector2{X:=255.0, Y:=50.0}, DefaultTint:=MakeColorFromHex("#FF000022")}#Red frame

                                                                                #Visible border for rarity layout
                                                                                overlay_slot:
                                                                                    VerticalAlignment:=vertical_alignment.Center
                                                                                    HorizontalAlignment:=horizontal_alignment.Center
                                                                                    Padding:=margin{Left:=0.0, Top:=0.0, Right:=0.0, Bottom:=0.0}
                                                                                    Widget:=texture_block{DefaultImage:=Textures.General.T_GenericBorderSquare3, DefaultDesiredSize:=vector2{X:=255.0, Y:=40.0}, DefaultTint:=MakeColorFromHex("#174583FF")}

                                                                                #Horizontal Stack box for rarity slots
                                                                                overlay_slot:
                                                                                    VerticalAlignment:=vertical_alignment.Fill
                                                                                    HorizontalAlignment:=horizontal_alignment.Fill
                                                                                    Padding:=margin{Left:=5.0, Top:=0.0, Right:=5.0, Bottom:=0.0}
                                                                                    Widget:=stack_box:
                                                                                        Orientation:=orientation.Horizontal
                                                                                        Slots:=array:
                                                                                            #Overlay for each rarity slots, including the pointing arrow
                                                                                            stack_box_slot:
                                                                                                VerticalAlignment:=vertical_alignment.Center
                                                                                                HorizontalAlignment:=horizontal_alignment.Center
                                                                                                Distribution:=option{1.0}
                                                                                                Padding:=margin{Left:=0.0, Top:=0.0, Right:=0.0, Bottom:=0.0}
                                                                                                Widget:=overlay:
                                                                                                    Slots:=array:
                                                                                                        #Makeshift size box
                                                                                                        overlay_slot:
                                                                                                            VerticalAlignment:=vertical_alignment.Center
                                                                                                            HorizontalAlignment:=horizontal_alignment.Center
                                                                                                            Padding:=margin{Left:=0.0, Top:=0.0, Right:=0.0, Bottom:=0.0}
                                                                                                            Widget:=texture_block{DefaultImage:=Textures.General.T_GenSquTrans, DefaultDesiredSize:=vector2{X:=30.0, Y:=40.0}, DefaultTint:=MakeColorFromHex("#0000FF08")}

                                                                                                        #Overlay for rarity border and texture
                                                                                                        overlay_slot:
                                                                                                            VerticalAlignment:=vertical_alignment.Center
                                                                                                            HorizontalAlignment:=horizontal_alignment.Center
                                                                                                            Padding:=margin{Left:=0.0, Top:=0.0, Right:=0.0, Bottom:=0.0}
                                                                                                            Widget:=overlay:
                                                                                                                Slots:=array:
                                                                                                                    
                                                                                                                    #Rarity texture border, when the rarity slot is empty
                                                                                                                    overlay_slot:
                                                                                                                        VerticalAlignment:=vertical_alignment.Center
                                                                                                                        HorizontalAlignment:=horizontal_alignment.Center
                                                                                                                        Padding:=margin{Left:=0.0, Top:=0.0, Right:=0.0, Bottom:=0.0}
                                                                                                                        Widget:=texture_block{DefaultImage:=Textures.General.T_GenericBorderSquare2, DefaultDesiredSize:=vector2{X:=30.0, Y:=30.0}, DefaultTint:=MakeColorFromHex("#174583FF")}

                                                                                                                    #Rarity texture
                                                                                                                    overlay_slot:
                                                                                                                        VerticalAlignment:=vertical_alignment.Center
                                                                                                                        HorizontalAlignment:=horizontal_alignment.Center
                                                                                                                        Padding:=margin{Left:=0.0, Top:=0.0, Right:=0.0, Bottom:=0.0}
                                                                                                                        Widget:=texture_block{DefaultImage:=Textures.General.T_GenericSquare, DefaultDesiredSize:=vector2{X:=30.0, Y:=30.0}, DefaultTint:=MakeColorFromHex("#979797FF")}
                                                                                                        
                                                                                                                    

                                                                                                        #Bracket
                                                                                                        overlay_slot:
                                                                                                            VerticalAlignment:=vertical_alignment.Fill
                                                                                                            HorizontalAlignment:=horizontal_alignment.Fill
                                                                                                            Padding:=margin{Left:=0.0, Top:=0.0, Right:=0.0, Bottom:=0.0}
                                                                                                            Widget:=stack_box:
                                                                                                                Orientation:=orientation.Horizontal
                                                                                                                #two instances of bracket texture to represent the range of rarities
                                                                                                                Slots:=array:

                                                                                                                    stack_box_slot:
                                                                                                                        VerticalAlignment:=vertical_alignment.Center
                                                                                                                        HorizontalAlignment:=horizontal_alignment.Left
                                                                                                                        Distribution:=option{1.0}
                                                                                                                        Padding:=margin{Left:=0.0, Top:=0.0, Right:=0.0, Bottom:=0.0}
                                                                                                                        Widget:=texture_block{DefaultImage:=Textures.General.T_RarityBracketLeft, DefaultDesiredSize:=vector2{X:=18.0, Y:=36.0}, DefaultTint:=MakeColorFromHex("#a3a6aaff")}

                                                                                                                    stack_box_slot:
                                                                                                                        VerticalAlignment:=vertical_alignment.Center
                                                                                                                        HorizontalAlignment:=horizontal_alignment.Right
                                                                                                                        Distribution:=option{1.0}
                                                                                                                        Padding:=margin{Left:=0.0, Top:=0.0, Right:=0.0, Bottom:=0.0}
                                                                                                                        Widget:=texture_block{DefaultImage:=Textures.General.T_RarityBracketRight, DefaultDesiredSize:=vector2{X:=18.0, Y:=36.0}, DefaultTint:=MakeColorFromHex("#a3a6aaff")}

                                                                                            stack_box_slot:
                                                                                                VerticalAlignment:=vertical_alignment.Center
                                                                                                HorizontalAlignment:=horizontal_alignment.Center
                                                                                                Distribution:=option{1.0}
                                                                                                Padding:=margin{Left:=0.0, Top:=0.0, Right:=0.0, Bottom:=0.0}
                                                                                                Widget:=overlay:
                                                                                                    Slots:=array:
                                                                                                        #Makeshift size box
                                                                                                        overlay_slot:
                                                                                                            VerticalAlignment:=vertical_alignment.Center
                                                                                                            HorizontalAlignment:=horizontal_alignment.Center
                                                                                                            Padding:=margin{Left:=0.0, Top:=0.0, Right:=0.0, Bottom:=0.0}
                                                                                                            Widget:=texture_block{DefaultImage:=Textures.General.T_GenSquTrans, DefaultDesiredSize:=vector2{X:=30.0, Y:=40.0}, DefaultTint:=MakeColorFromHex("#0000FF08")}

                                                                                                        #Overlay for rarity border and texture
                                                                                                        overlay_slot:
                                                                                                            VerticalAlignment:=vertical_alignment.Center
                                                                                                            HorizontalAlignment:=horizontal_alignment.Center
                                                                                                            Padding:=margin{Left:=0.0, Top:=0.0, Right:=0.0, Bottom:=0.0}
                                                                                                            Widget:=overlay:
                                                                                                                Slots:=array:
                                                                                                                    
                                                                                                                    #Rarity texture border, when the rarity slot is empty
                                                                                                                    overlay_slot:
                                                                                                                        VerticalAlignment:=vertical_alignment.Center
                                                                                                                        HorizontalAlignment:=horizontal_alignment.Center
                                                                                                                        Padding:=margin{Left:=0.0, Top:=0.0, Right:=0.0, Bottom:=0.0}
                                                                                                                        Widget:=texture_block{DefaultImage:=Textures.General.T_GenericBorderSquare2, DefaultDesiredSize:=vector2{X:=30.0, Y:=30.0}, DefaultTint:=MakeColorFromHex("#174583FF")}

                                                                                                                    #Rarity texture
                                                                                                                    overlay_slot:
                                                                                                                        VerticalAlignment:=vertical_alignment.Center
                                                                                                                        HorizontalAlignment:=horizontal_alignment.Center
                                                                                                                        Padding:=margin{Left:=0.0, Top:=0.0, Right:=0.0, Bottom:=0.0}
                                                                                                                        Widget:=texture_block{DefaultImage:=Textures.General.T_GenericSquare, DefaultDesiredSize:=vector2{X:=30.0, Y:=30.0}, DefaultTint:=MakeColorFromHex("#4ccc2aFF")}

                                                                                                        #Bracket
                                                                                                        overlay_slot:
                                                                                                            VerticalAlignment:=vertical_alignment.Fill
                                                                                                            HorizontalAlignment:=horizontal_alignment.Fill
                                                                                                            Padding:=margin{Left:=0.0, Top:=0.0, Right:=0.0, Bottom:=0.0}
                                                                                                            Widget:=stack_box:
                                                                                                                Orientation:=orientation.Horizontal
                                                                                                                #two instances of bracket texture to represent the range of rarities
                                                                                                                Slots:=array:

                                                                                                                    stack_box_slot:
                                                                                                                        VerticalAlignment:=vertical_alignment.Center
                                                                                                                        HorizontalAlignment:=horizontal_alignment.Left
                                                                                                                        Distribution:=option{1.0}
                                                                                                                        Padding:=margin{Left:=0.0, Top:=0.0, Right:=0.0, Bottom:=0.0}
                                                                                                                        Widget:=texture_block{DefaultImage:=Textures.General.T_RarityBracketLeft, DefaultDesiredSize:=vector2{X:=18.0, Y:=36.0}, DefaultTint:=MakeColorFromHex("#a3a6aaff")}

                                                                                                                    stack_box_slot:
                                                                                                                        VerticalAlignment:=vertical_alignment.Center
                                                                                                                        HorizontalAlignment:=horizontal_alignment.Right
                                                                                                                        Distribution:=option{1.0}
                                                                                                                        Padding:=margin{Left:=0.0, Top:=0.0, Right:=0.0, Bottom:=0.0}
                                                                                                                        Widget:=texture_block{DefaultImage:=Textures.General.T_RarityBracketRight, DefaultDesiredSize:=vector2{X:=18.0, Y:=36.0}, DefaultTint:=MakeColorFromHex("#a3a6aaff")}

                                                                                            stack_box_slot:
                                                                                                VerticalAlignment:=vertical_alignment.Center
                                                                                                HorizontalAlignment:=horizontal_alignment.Center
                                                                                                Distribution:=option{1.0}
                                                                                                Padding:=margin{Left:=0.0, Top:=0.0, Right:=0.0, Bottom:=0.0}
                                                                                                Widget:=overlay:
                                                                                                    Slots:=array:
                                                                                                        #Makeshift size box
                                                                                                        overlay_slot:
                                                                                                            VerticalAlignment:=vertical_alignment.Center
                                                                                                            HorizontalAlignment:=horizontal_alignment.Center
                                                                                                            Padding:=margin{Left:=0.0, Top:=0.0, Right:=0.0, Bottom:=0.0}
                                                                                                            Widget:=texture_block{DefaultImage:=Textures.General.T_GenSquTrans, DefaultDesiredSize:=vector2{X:=30.0, Y:=40.0}, DefaultTint:=MakeColorFromHex("#0000FF08")}

                                                                                                        #Overlay for rarity border and texture
                                                                                                        overlay_slot:
                                                                                                            VerticalAlignment:=vertical_alignment.Center
                                                                                                            HorizontalAlignment:=horizontal_alignment.Center
                                                                                                            Padding:=margin{Left:=0.0, Top:=0.0, Right:=0.0, Bottom:=0.0}
                                                                                                            Widget:=overlay:
                                                                                                                Slots:=array:
                                                                                                                                                                                                                
                                                                                                                    #Rarity texture border, when the rarity slot is empty
                                                                                                                    overlay_slot:
                                                                                                                        VerticalAlignment:=vertical_alignment.Center
                                                                                                                        HorizontalAlignment:=horizontal_alignment.Center
                                                                                                                        Padding:=margin{Left:=0.0, Top:=0.0, Right:=0.0, Bottom:=0.0}
                                                                                                                        Widget:=texture_block{DefaultImage:=Textures.General.T_GenericBorderSquare2, DefaultDesiredSize:=vector2{X:=30.0, Y:=30.0}, DefaultTint:=MakeColorFromHex("#174583FF")}
                                                                                                                                                                                                                            
                                                                                                                    #Rarity texture
                                                                                                                    overlay_slot:
                                                                                                                        VerticalAlignment:=vertical_alignment.Bottom
                                                                                                                        HorizontalAlignment:=horizontal_alignment.Center
                                                                                                                        Padding:=margin{Left:=0.0, Top:=0.0, Right:=0.0, Bottom:=0.0}
                                                                                                                        Widget:=texture_block{DefaultImage:=Textures.General.T_GenericSquare, DefaultDesiredSize:=vector2{X:=30.0, Y:=30.0}, DefaultTint:=MakeColorFromHex("#183379FF")}
                                                                                                                    
                                                                                                        #Bracket
                                                                                                        overlay_slot:
                                                                                                            VerticalAlignment:=vertical_alignment.Fill
                                                                                                            HorizontalAlignment:=horizontal_alignment.Fill
                                                                                                            Padding:=margin{Left:=0.0, Top:=0.0, Right:=0.0, Bottom:=0.0}
                                                                                                            Widget:=stack_box:
                                                                                                                Orientation:=orientation.Horizontal
                                                                                                                #two instances of bracket texture to represent the range of rarities
                                                                                                                Slots:=array:

                                                                                                                    stack_box_slot:
                                                                                                                        VerticalAlignment:=vertical_alignment.Center
                                                                                                                        HorizontalAlignment:=horizontal_alignment.Left
                                                                                                                        Distribution:=option{1.0}
                                                                                                                        Padding:=margin{Left:=0.0, Top:=0.0, Right:=0.0, Bottom:=0.0}
                                                                                                                        Widget:=texture_block{DefaultImage:=Textures.General.T_RarityBracketLeft, DefaultDesiredSize:=vector2{X:=18.0, Y:=36.0}, DefaultTint:=MakeColorFromHex("#a3a6aaff")}

                                                                                                                    stack_box_slot:
                                                                                                                        VerticalAlignment:=vertical_alignment.Center
                                                                                                                        HorizontalAlignment:=horizontal_alignment.Right
                                                                                                                        Distribution:=option{1.0}
                                                                                                                        Padding:=margin{Left:=0.0, Top:=0.0, Right:=0.0, Bottom:=0.0}
                                                                                                                        Widget:=texture_block{DefaultImage:=Textures.General.T_RarityBracketRight, DefaultDesiredSize:=vector2{X:=18.0, Y:=36.0}, DefaultTint:=MakeColorFromHex("#a3a6aaff")}

                                                                                            stack_box_slot:
                                                                                                VerticalAlignment:=vertical_alignment.Center
                                                                                                HorizontalAlignment:=horizontal_alignment.Center
                                                                                                Distribution:=option{1.0}
                                                                                                Padding:=margin{Left:=0.0, Top:=0.0, Right:=0.0, Bottom:=0.0}
                                                                                                Widget:=overlay:
                                                                                                    Slots:=array:
                                                                                                        #Makeshift size box
                                                                                                        overlay_slot:
                                                                                                            VerticalAlignment:=vertical_alignment.Center
                                                                                                            HorizontalAlignment:=horizontal_alignment.Center
                                                                                                            Padding:=margin{Left:=0.0, Top:=0.0, Right:=0.0, Bottom:=0.0}
                                                                                                            Widget:=texture_block{DefaultImage:=Textures.General.T_GenSquTrans, DefaultDesiredSize:=vector2{X:=30.0, Y:=40.0}, DefaultTint:=MakeColorFromHex("#0000FF08")}

                                                                                                        #Overlay for rarity border and texture
                                                                                                        overlay_slot:
                                                                                                            VerticalAlignment:=vertical_alignment.Center
                                                                                                            HorizontalAlignment:=horizontal_alignment.Center
                                                                                                            Padding:=margin{Left:=0.0, Top:=0.0, Right:=0.0, Bottom:=0.0}
                                                                                                            Widget:=overlay:
                                                                                                                Slots:=array:

                                                                                                                    #Rarity texture border, when the rarity slot is empty
                                                                                                                    overlay_slot:
                                                                                                                        VerticalAlignment:=vertical_alignment.Center
                                                                                                                        HorizontalAlignment:=horizontal_alignment.Center
                                                                                                                        Padding:=margin{Left:=0.0, Top:=0.0, Right:=0.0, Bottom:=0.0}
                                                                                                                        Widget:=texture_block{DefaultImage:=Textures.General.T_GenericBorderSquare2, DefaultDesiredSize:=vector2{X:=30.0, Y:=30.0}, DefaultTint:=MakeColorFromHex("#174583FF")}
                                                                                                                    
                                                                                                                    #Rarity texture
                                                                                                                    overlay_slot:
                                                                                                                        VerticalAlignment:=vertical_alignment.Center
                                                                                                                        HorizontalAlignment:=horizontal_alignment.Center
                                                                                                                        Padding:=margin{Left:=0.0, Top:=0.0, Right:=0.0, Bottom:=0.0}
                                                                                                                        Widget:=texture_block{DefaultImage:=Textures.General.T_GenericSquare, DefaultDesiredSize:=vector2{X:=30.0, Y:=30.0}, DefaultTint:=MakeColorFromHex("#5b1e79FF")}
                                                                                                                    
                                                                                                        #Bracket
                                                                                                        overlay_slot:
                                                                                                            VerticalAlignment:=vertical_alignment.Fill
                                                                                                            HorizontalAlignment:=horizontal_alignment.Fill
                                                                                                            Padding:=margin{Left:=0.0, Top:=0.0, Right:=0.0, Bottom:=0.0}
                                                                                                            Widget:=stack_box:
                                                                                                                Orientation:=orientation.Horizontal
                                                                                                                #two instances of bracket texture to represent the range of rarities
                                                                                                                Slots:=array:

                                                                                                                    stack_box_slot:
                                                                                                                        VerticalAlignment:=vertical_alignment.Center
                                                                                                                        HorizontalAlignment:=horizontal_alignment.Left
                                                                                                                        Distribution:=option{1.0}
                                                                                                                        Padding:=margin{Left:=0.0, Top:=0.0, Right:=0.0, Bottom:=0.0}
                                                                                                                        Widget:=texture_block{DefaultImage:=Textures.General.T_RarityBracketLeft, DefaultDesiredSize:=vector2{X:=18.0, Y:=36.0}, DefaultTint:=MakeColorFromHex("#a3a6aaff")}

                                                                                                                    stack_box_slot:
                                                                                                                        VerticalAlignment:=vertical_alignment.Center
                                                                                                                        HorizontalAlignment:=horizontal_alignment.Right
                                                                                                                        Distribution:=option{1.0}
                                                                                                                        Padding:=margin{Left:=0.0, Top:=0.0, Right:=0.0, Bottom:=0.0}
                                                                                                                        Widget:=texture_block{DefaultImage:=Textures.General.T_RarityBracketRight, DefaultDesiredSize:=vector2{X:=18.0, Y:=36.0}, DefaultTint:=MakeColorFromHex("#a3a6aaff")}

                                                                                            stack_box_slot:
                                                                                                VerticalAlignment:=vertical_alignment.Center
                                                                                                HorizontalAlignment:=horizontal_alignment.Center
                                                                                                Distribution:=option{1.0}
                                                                                                Padding:=margin{Left:=0.0, Top:=0.0, Right:=0.0, Bottom:=0.0}
                                                                                                Widget:=overlay:
                                                                                                    Slots:=array:
                                                                                                        #Makeshift size box
                                                                                                        overlay_slot:
                                                                                                            VerticalAlignment:=vertical_alignment.Center
                                                                                                            HorizontalAlignment:=horizontal_alignment.Center
                                                                                                            Padding:=margin{Left:=0.0, Top:=0.0, Right:=0.0, Bottom:=0.0}
                                                                                                            Widget:=texture_block{DefaultImage:=Textures.General.T_GenSquTrans, DefaultDesiredSize:=vector2{X:=30.0, Y:=40.0}, DefaultTint:=MakeColorFromHex("#0000FF08")}

                                                                                                        #Overlay for rarity border and texture
                                                                                                        overlay_slot:
                                                                                                            VerticalAlignment:=vertical_alignment.Center
                                                                                                            HorizontalAlignment:=horizontal_alignment.Center
                                                                                                            Padding:=margin{Left:=0.0, Top:=0.0, Right:=0.0, Bottom:=0.0}
                                                                                                            Widget:=overlay:
                                                                                                                Slots:=array:

                                                                                                                    #Rarity texture border, when the rarity slot is empty
                                                                                                                    overlay_slot:
                                                                                                                        VerticalAlignment:=vertical_alignment.Center
                                                                                                                        HorizontalAlignment:=horizontal_alignment.Center
                                                                                                                        Padding:=margin{Left:=0.0, Top:=0.0, Right:=0.0, Bottom:=0.0}
                                                                                                                        Widget:=texture_block{DefaultImage:=Textures.General.T_GenericBorderSquare2, DefaultDesiredSize:=vector2{X:=30.0, Y:=30.0}, DefaultTint:=MakeColorFromHex("#174583FF")}
                                                                                                                    
                                                                                                                    #Rarity texture
                                                                                                                    overlay_slot:
                                                                                                                        VerticalAlignment:=vertical_alignment.Center
                                                                                                                        HorizontalAlignment:=horizontal_alignment.Center
                                                                                                                        Padding:=margin{Left:=0.0, Top:=0.0, Right:=0.0, Bottom:=0.0}
                                                                                                                        Widget:=texture_block{DefaultImage:=Textures.General.T_GenericSquare, DefaultDesiredSize:=vector2{X:=30.0, Y:=30.0}, DefaultTint:=MakeColorFromHex("#674f28FF")}

                                                                                                        #Bracket
                                                                                                        overlay_slot:
                                                                                                            VerticalAlignment:=vertical_alignment.Fill
                                                                                                            HorizontalAlignment:=horizontal_alignment.Fill
                                                                                                            Padding:=margin{Left:=0.0, Top:=0.0, Right:=0.0, Bottom:=0.0}
                                                                                                            Widget:=stack_box:
                                                                                                                Orientation:=orientation.Horizontal
                                                                                                                #two instances of bracket texture to represent the range of rarities
                                                                                                                Slots:=array:

                                                                                                                    stack_box_slot:
                                                                                                                        VerticalAlignment:=vertical_alignment.Center
                                                                                                                        HorizontalAlignment:=horizontal_alignment.Left
                                                                                                                        Distribution:=option{1.0}
                                                                                                                        Padding:=margin{Left:=0.0, Top:=0.0, Right:=0.0, Bottom:=0.0}
                                                                                                                        Widget:=texture_block{DefaultImage:=Textures.General.T_RarityBracketLeft, DefaultDesiredSize:=vector2{X:=18.0, Y:=36.0}, DefaultTint:=MakeColorFromHex("#a3a6aaff")}

                                                                                                                    stack_box_slot:
                                                                                                                        VerticalAlignment:=vertical_alignment.Center
                                                                                                                        HorizontalAlignment:=horizontal_alignment.Right
                                                                                                                        Distribution:=option{1.0}
                                                                                                                        Padding:=margin{Left:=0.0, Top:=0.0, Right:=0.0, Bottom:=0.0}
                                                                                                                        Widget:=texture_block{DefaultImage:=Textures.General.T_RarityBracketRight, DefaultDesiredSize:=vector2{X:=18.0, Y:=36.0}, DefaultTint:=MakeColorFromHex("#a3a6aaff")}

                                                                                            stack_box_slot:
                                                                                                VerticalAlignment:=vertical_alignment.Center
                                                                                                HorizontalAlignment:=horizontal_alignment.Center
                                                                                                Distribution:=option{1.0}
                                                                                                Padding:=margin{Left:=0.0, Top:=0.0, Right:=0.0, Bottom:=0.0}
                                                                                                Widget:=overlay:
                                                                                                    Slots:=array:
                                                                                                        #Makeshift size box
                                                                                                        overlay_slot:
                                                                                                            VerticalAlignment:=vertical_alignment.Center
                                                                                                            HorizontalAlignment:=horizontal_alignment.Center
                                                                                                            Padding:=margin{Left:=0.0, Top:=0.0, Right:=0.0, Bottom:=0.0}
                                                                                                            Widget:=texture_block{DefaultImage:=Textures.General.T_GenSquTrans, DefaultDesiredSize:=vector2{X:=30.0, Y:=40.0}, DefaultTint:=MakeColorFromHex("#0000FF08")}

                                                                                                        #Overlay for rarity border and texture
                                                                                                        overlay_slot:
                                                                                                            VerticalAlignment:=vertical_alignment.Center
                                                                                                            HorizontalAlignment:=horizontal_alignment.Center
                                                                                                            Padding:=margin{Left:=0.0, Top:=0.0, Right:=0.0, Bottom:=0.0}
                                                                                                            Widget:=overlay:
                                                                                                                Slots:=array:
                                                                                                        
                                                                                                                    #Rarity texture border, when the rarity slot is empty
                                                                                                                    overlay_slot:
                                                                                                                        VerticalAlignment:=vertical_alignment.Center
                                                                                                                        HorizontalAlignment:=horizontal_alignment.Center
                                                                                                                        Padding:=margin{Left:=0.0, Top:=0.0, Right:=0.0, Bottom:=0.0}
                                                                                                                        Widget:=texture_block{DefaultImage:=Textures.General.T_GenericBorderSquare2, DefaultDesiredSize:=vector2{X:=30.0, Y:=30.0}, DefaultTint:=MakeColorFromHex("#174583FF")}
                                                                                                                    
                                                                                                                    #Rarity texture
                                                                                                                    overlay_slot:
                                                                                                                        VerticalAlignment:=vertical_alignment.Center
                                                                                                                        HorizontalAlignment:=horizontal_alignment.Center
                                                                                                                        Padding:=margin{Left:=0.0, Top:=0.0, Right:=0.0, Bottom:=0.0}
                                                                                                                        Widget:=texture_block{DefaultImage:=Textures.General.T_GenericSquare, DefaultDesiredSize:=vector2{X:=30.0, Y:=30.0}, DefaultTint:=MakeColorFromHex("#78711fFF")}
                                                                                                                    
                                                                                                        #Bracket
                                                                                                        overlay_slot:
                                                                                                            VerticalAlignment:=vertical_alignment.Fill
                                                                                                            HorizontalAlignment:=horizontal_alignment.Fill
                                                                                                            Padding:=margin{Left:=0.0, Top:=0.0, Right:=0.0, Bottom:=0.0}
                                                                                                            Widget:=stack_box:
                                                                                                                Orientation:=orientation.Horizontal
                                                                                                                #two instances of bracket texture to represent the range of rarities
                                                                                                                Slots:=array:

                                                                                                                    stack_box_slot:
                                                                                                                        VerticalAlignment:=vertical_alignment.Center
                                                                                                                        HorizontalAlignment:=horizontal_alignment.Left
                                                                                                                        Distribution:=option{1.0}
                                                                                                                        Padding:=margin{Left:=0.0, Top:=0.0, Right:=0.0, Bottom:=0.0}
                                                                                                                        Widget:=texture_block{DefaultImage:=Textures.General.T_RarityBracketLeft, DefaultDesiredSize:=vector2{X:=18.0, Y:=36.0}, DefaultTint:=MakeColorFromHex("#a3a6aaff")}

                                                                                                                    stack_box_slot:
                                                                                                                        VerticalAlignment:=vertical_alignment.Center
                                                                                                                        HorizontalAlignment:=horizontal_alignment.Right
                                                                                                                        Distribution:=option{1.0}
                                                                                                                        Padding:=margin{Left:=0.0, Top:=0.0, Right:=0.0, Bottom:=0.0}
                                                                                                                        Widget:=texture_block{DefaultImage:=Textures.General.T_RarityBracketRight, DefaultDesiredSize:=vector2{X:=18.0, Y:=36.0}, DefaultTint:=MakeColorFromHex("#a3a6aaff")}

                    #overlay_slot:
                        #VerticalAlignment:=vertical_alignment.Fill
                        #HorizontalAlignment:=horizontal_alignment.Fill
                        #Padding:=margin{Left:=0.0, Top:=0.0, Right:=0.0, Bottom:=0.0}
                        #Widget:=texture_block{DefaultImage:=Textures.General.T_LevelPointer, DefaultDesiredSize:=vector2{X:=20.0, Y:=20.0}, DefaultTint:=MakeColorFromHex("#7de2FFFF")}



            #Build of the screen
            HeaderCanvas:canvas = canvas:
                Slots:=array:
                    #Defines the parameters for the canvas' slot
                    canvas_slot:
                        Anchors := anchors{Minimum := vector2{X := 0.0, Y := 0.0}, Maximum := vector2{X := 1.0, Y := 1.0}}
                        Alignment := vector2{X := 0.5, Y := 0.5}
                        Widget:=overlay:
                            Slots:=array:
                                overlay_slot:
                                    VerticalAlignment:=vertical_alignment.Center
                                    HorizontalAlignment:=horizontal_alignment.Center
                                    Padding:=margin{Left:=0.0, Top:=0.0, Right:=0.0, Bottom:=0.0}
                                    #Screen background
                                    Widget:=texture_block{DefaultImage:=Textures.General.T_Background, DefaultDesiredSize:=vector2{X:=1910.0, Y:=1080.0}}

                                overlay_slot:
                                    VerticalAlignment:=vertical_alignment.Center
                                    HorizontalAlignment:=horizontal_alignment.Center
                                    Padding:=margin{Left:=0.0, Top:=0.0, Right:=0.0, Bottom:=0.0}
                                    #Header
                                    Widget:=stack_box:
                                        Orientation:=orientation.Vertical
                                        Slots:=array:

                                            stack_box_slot:
                                                VerticalAlignment:=vertical_alignment.Top
                                                HorizontalAlignment:=horizontal_alignment.Center
                                                Padding:=margin{Left:=0.0, Top:=0.0, Right:=0.0, Bottom:=0.0}
                                                Widget:=HeaderStackBox

                                            stack_box_slot:
                                                
                                                VerticalAlignment:=vertical_alignment.Center
                                                HorizontalAlignment:=horizontal_alignment.Center
                                                Padding:=margin{Left:=0.0, Top:=0.0, Right:=0.0, Bottom:=0.0}
                                                Widget:=overlay:
                                                    Slots:=array:

                                                        #Makeshift size box for the lower content
                                                        overlay_slot:
                                                            VerticalAlignment:=vertical_alignment.Center
                                                            HorizontalAlignment:=horizontal_alignment.Center
                                                            Padding:=margin{Left:=0.0, Top:=0.0, Right:=0.0, Bottom:=0.0}
                                                            Widget:=texture_block{DefaultImage:=Textures.General.T_GenSquTrans, DefaultDesiredSize:=vector2{X:=1750.0, Y:=900.0}}

                                                        #Horizontal stack box for content
                                                        overlay_slot:
                                                            VerticalAlignment:=vertical_alignment.Top
                                                            HorizontalAlignment:=horizontal_alignment.Fill
                                                            Padding:=margin{Left:=0.0, Top:=0.0, Right:=0.0, Bottom:=0.0}
                                                            Widget:=stack_box:
                                                                Orientation:=orientation.Horizontal
                                                                Slots:=array:

                                                                    #First column of boosters
                                                                    stack_box_slot:
                                                                        VerticalAlignment:=vertical_alignment.Center
                                                                        HorizontalAlignment:=horizontal_alignment.Left
                                                                        Distribution:=option{1.0}
                                                                        Padding:=margin{Left:=0.0, Top:=0.0, Right:=0.0, Bottom:=0.0}
                                                                        Widget:= stack_box:
                                                                            Orientation:=orientation.Vertical
                                                                            Slots:=array:
                                                                                
                                                                                stack_box_slot:
                                                                                    VerticalAlignment:=vertical_alignment.Center
                                                                                    HorizontalAlignment:=horizontal_alignment.Center
                                                                                    Padding:=margin{Left:=0.0, Top:=8.0, Right:=0.0, Bottom:=0.0}
                                                                                    #Header
                                                                                    Widget:=LevelRow0

                                                                                stack_box_slot:
                                                                                    VerticalAlignment:=vertical_alignment.Center
                                                                                    HorizontalAlignment:=horizontal_alignment.Center
                                                                                    Padding:=margin{Left:=0.0, Top:=8.0, Right:=0.0, Bottom:=0.0}
                                                                                    #Header
                                                                                    Widget:=LevelRow1

                                                                                stack_box_slot:
                                                                                    VerticalAlignment:=vertical_alignment.Center
                                                                                    HorizontalAlignment:=horizontal_alignment.Center
                                                                                    Padding:=margin{Left:=0.0, Top:=8.0, Right:=0.0, Bottom:=0.0}
                                                                                    #Header
                                                                                    Widget:=LevelRow2

                                                                                stack_box_slot:
                                                                                    VerticalAlignment:=vertical_alignment.Center
                                                                                    HorizontalAlignment:=horizontal_alignment.Center
                                                                                    Padding:=margin{Left:=0.0, Top:=8.0, Right:=0.0, Bottom:=0.0}
                                                                                    #Header
                                                                                    Widget:=LevelRow3

                                                                                stack_box_slot:
                                                                                    VerticalAlignment:=vertical_alignment.Center
                                                                                    HorizontalAlignment:=horizontal_alignment.Center
                                                                                    Padding:=margin{Left:=0.0, Top:=8.0, Right:=0.0, Bottom:=0.0}
                                                                                    #Header
                                                                                    Widget:=LevelRow4

                                                                                stack_box_slot:
                                                                                    VerticalAlignment:=vertical_alignment.Center
                                                                                    HorizontalAlignment:=horizontal_alignment.Center
                                                                                    Padding:=margin{Left:=0.0, Top:=8.0, Right:=0.0, Bottom:=0.0}
                                                                                    #Header
                                                                                    Widget:=LevelRow5

                                                                                stack_box_slot:
                                                                                    VerticalAlignment:=vertical_alignment.Center
                                                                                    HorizontalAlignment:=horizontal_alignment.Center
                                                                                    Padding:=margin{Left:=0.0, Top:=8.0, Right:=0.0, Bottom:=0.0}
                                                                                    #Header
                                                                                    Widget:=LevelRow6

                                                                                

                                                                    #Second column of boosters
                                                                    stack_box_slot:
                                                                        VerticalAlignment:=vertical_alignment.Center
                                                                        HorizontalAlignment:=horizontal_alignment.Right
                                                                        Padding:=margin{Left:=0.0, Top:=0.0, Right:=0.0, Bottom:=0.0}
                                                                        Widget:= stack_box:
                                                                            Orientation:=orientation.Vertical
                                                                            Slots:=array:
                                                                                
                                                                                stack_box_slot:
                                                                                    VerticalAlignment:=vertical_alignment.Center
                                                                                    HorizontalAlignment:=horizontal_alignment.Center
                                                                                    Padding:=margin{Left:=0.0, Top:=8.0, Right:=0.0, Bottom:=0.0}
                                                                                    #Header
                                                                                    Widget:=LoadoutCard

                                                                               
                                                                                            
                                                                                
                                                                                                            


                                                                                                                                                                                                                    
                                                                                                                    

            if(PlayerUI:=GetPlayerUI[Player]):
                PlayerUI.AddWidget(HeaderCanvas, player_ui_slot{InputMode:=ui_input_mode.All})