using { /UnrealEngine.com/Temporary/SpatialMath}
using { /Verse.org/Simulation }
using { /Verse.org/Random }
using { /Fortnite.com/Devices}
using { /UnrealEngine.com/Temporary/UI }

<# 
General functions and definitions that can be used in all verse classes
#>

INDEX_NONE:int = -1
VALUE_NONE:int = -1
FVALUE_NONE:float = -1.0

ZERO_VECTOR2:vector2 = vector2{X:=0.0, Y:=0.0}
ONE_VECTOR2:vector2 = vector2{X:=1.0, Y:=1.0}

StringToMessage<localizes><public>(String:string)<computes>:message = "{String}"
AgentToMessage<localizes><public>(Agent:agent)<computes>:message = "{Agent}"

# CAREFUL: This function cannot be used with float values bigger than int
GetRoundedString<public>(Input : float, NumDigitsAfterDecimal : int) : string =
    if:
        Multiplier := Pow(10.0, NumDigitsAfterDecimal * 1.0)
        RoundedValue := RoundedTickRate := float[Round[Input*Multiplier]*1.0] / Multiplier
        BeforeDecimal := Floor[RoundedValue]
        AfterDecimal := Abs(Round[(RoundedValue - BeforeDecimal*1.0)*Multiplier])
        var AfterDecimalString : string = ToString(AfterDecimal)

        #pad with leading zeroes
        for (It := 0..(NumDigitsAfterDecimal-AfterDecimalString.Length-1)):
            set AfterDecimalString = array{'0'} + AfterDecimalString
    then:
        "{BeforeDecimal}.{AfterDecimalString}"
    else:
        ToString(Input)

# CAREFUL: This function cannot be used with float values bigger than int
TruncateNumber(Number:float, Decimals:int)<transacts>:float =
    if:
        Decimals >= 0
        Multiplier := Pow(10.0, Decimals * 1.0)
        RoundValue := Round[Number * Multiplier] * 1.0
    then:
        RoundValue / Multiplier
    else:
        Number

FormatNumber<public>(Number:int):string =
    var NumberString:string = "{Number}"
    var FinalFormat:string = ""

    for (Index:=0..(NumberString.Length-1)):
        if (Mod[NumberString.Length-Index,3] = 0 and Index<>0):
            set FinalFormat += ","

        if (Char:=NumberString[Index]):
            set FinalFormat += "{Char}"

    return FinalFormat

ToTruncatedString<public>(Number:float, Decimals:int):[]char =
    var Str:[]char = ToString(Number)
    if:
        DotIndex := Str.Find['.']
        StopIndex := if (Decimals > 0) then Min(DotIndex+Decimals+1, Str.Length) else DotIndex
        if (Truncated := Str.Slice[0, StopIndex]):
            set Str = Truncated
    Str

<# Makes a float into string, with decimals if != than zero otherwise no decimals#>
ToSmartString<public>(Number:float, MaxDecimals:int)<transacts>:[]char =
    var Str:[]char = ToString(Number)
    if (DotIndex:=Str.Find['.']):
        MaxIndex:int = if (MaxDecimals > 0) then Min(DotIndex+MaxDecimals, Str.Length-1) else DotIndex-1

        var StopIndex:int = DotIndex-1
        var bAllZero:logic = true
        for (Index:=(DotIndex+1)..MaxIndex):
            if (Str[Index] <> '0'):
                set bAllZero = false
                set StopIndex = Index

        if (bAllZero?):
            set StopIndex = DotIndex-1
        
        if (StopIndex < Str.Length, Truncated := Str.Slice[0, StopIndex+1]):
            set Str = Truncated
    Str

# Returns a string highlighted with a different color
# TODO I did not find a way with RichTextBlock, I was trying <EM>3</>, it did not work. I will leave the function just in case I find a way @mbohorquez 08/1/2024
AsHighlightedString(Number:float)<transacts>:string = "{ToSmartString(Number, 2)}"
AsHighlightedString(Number:int)<transacts>:string = "{Number}"

TimerToString(Minutes:int, Seconds:int)<computes>:string =
    # If Seconds < 10, then add a 0 in front so it displays 0# instead of #
    SecondsAsString := if (Seconds < 10) then Join(array{ToString(0), ToString(Seconds)}, "") else ToString(Seconds)
    "{Minutes}:{SecondsAsString}"

TimeInSecondsToString(Seconds:int)<computes>:string =
    TimerToString(SecondsToMinutes(Seconds), GetRemainderSecondsFromSecondsToMinutes(Seconds))

SecondsToMinutes(Seconds:int)<computes>:int =
    Floor(Seconds / 60 or Err())

# Returns the reminder from seconds when changing to minutes
GetRemainderSecondsFromSecondsToMinutes(Seconds:int)<computes>:int =
    Mod[Seconds, 60] or 0

DurationToString(Seconds:int)<computes>: string =
    if (Minutes:=SecondsToMinutes(Seconds), Minutes > 0):
        return "{Minutes}min"
    return "{Seconds}s"

<# 
Array
#>

# Returns a random type from list ignoring some, won't ignore if not more available
GetRandomInArray(Array:[]t, TryIgnore:[]t where t:subtype(comparable))<transacts><decides>:t =
    var MaybeValue:?t = false

    if (Array.Length > 0):
        var RandomIn:[]t = for (Value:Array, not TryIgnore.Find[Value]):
            Value
        
        if (RandomIn.Length <= 0):
            # No unique, then just select one randomly from all
            set RandomIn = Array

        RandomIndex:=GetRandomInt(0, RandomIn.Length-1)
        if (RandomValue:=RandomIn[RandomIndex]):
            set MaybeValue = option{RandomValue}
        
    MaybeValue?

GetIndexClampedInArray(Array:[]t, Index:int where t:type)<computes>: int =
    if (Index >= Array.Length) then (Array.Length-1) else if (Index < 0) then 0 else Index

<# 
INCREMENTAL VALUE
#>
increment_type := enum:
    Additive
    Exponential # For number < 1.000 it removes decimals = This was made for a quick fix in prices
    Multiply
    ExponentialRaw # Same as exponential but no round

incremental_value := struct<concrete>:
    @editable # Initial value
    Initial:float = 0.0

    @editable # Max increments this value can have, if undefined MaybeMaximum will be used
    MaybeMaxIncrements:?int = false

    @editable # Maximum value, if false then INF
    MaybeMaximum:?float = false

    @editable
    IncrementType:increment_type = increment_type.Additive

    @editable
    IncrementValue:float = 1.0

# Fails if Level gets out of max value, returns value in level if success
# Do not use this to return value in clamped level, use `GetValueInClampedLevel()`
(IncrementalValue:incremental_value).HasLevel(Level:int)<decides><transacts>: float =
    var MaybeValueInLevel:?float = false

    if (Level >= 0):
        ValueInLevel:=IncrementalValue.GetValueInLevel(Level)

        if (MaxIncrements:=IncrementalValue.MaybeMaxIncrements?):
            if (MaxIncrements >= 0 and Level <= MaxIncrements):
                set MaybeValueInLevel = option{ValueInLevel}
        else if (Maximum:=IncrementalValue.MaybeMaximum?):
            # Only set the value when success
            if (ValueInLevel <= Maximum):
                set MaybeValueInLevel = option{ValueInLevel}
        else: # INF
            set MaybeValueInLevel = option{ValueInLevel}
    
    MaybeValueInLevel?

# Get the value in a specific level. NOTE: If you update the formulas, you need to update behaviour in `GetMaxLevel()` too
(IncrementalValue:incremental_value).GetValueInLevel(Level:int)<computes><reads>: float =
    case(IncrementalValue.IncrementType):
        increment_type.Additive => IncrementalValue.Initial + (IncrementalValue.IncrementValue * Level)
        
        increment_type.Exponential => 
            Value:=IncrementalValue.Initial * Pow(IncrementalValue.IncrementValue, Level*1.0)
            # When number is less than 1.000 (1K) remove decimals, we do not remove always before this can crash if the float is bigger than available int
            if (Value < 1000.0, IntValue:int = Int[Value]):
                IntValue*1.0
            else:
                Value

        increment_type.ExponentialRaw => IncrementalValue.Initial * Pow(IncrementalValue.IncrementValue, Level*1.0)

        increment_type.Multiply => IncrementalValue.Initial + (IncrementalValue.Initial * IncrementalValue.IncrementValue * Level)

# Returns the max level, false if INF. NOTE: If you update the formulas, you need to update behaviour in `GetValueInLevel()` too
(IncrementalValue:incremental_value).GetMaybeMaxLevel()<decides><transacts>: ?int =
    var MaybeMaxLevel:?int = false

    if (MaxIncrements:=IncrementalValue.MaybeMaxIncrements?):
        if (MaxIncrements >= 0):
            set MaybeMaxLevel = option{MaxIncrements}
    else if (Maximum:=IncrementalValue.MaybeMaximum?):
        var MaybeLevelResult:?float = false
        
        case(IncrementalValue.IncrementType):
            increment_type.Additive =>
                set MaybeLevelResult = option:
                    (Maximum - IncrementalValue.Initial) / IncrementalValue.IncrementValue
            
            increment_type.Exponential => 
                set MaybeLevelResult = option:
                    Log(IncrementalValue.IncrementValue, (Maximum / IncrementalValue.Initial))

            increment_type.ExponentialRaw => 
                set MaybeLevelResult = option:
                    Log(IncrementalValue.IncrementValue, (Maximum / IncrementalValue.Initial))

            increment_type.Multiply => 
                set MaybeLevelResult = option:
                    (Maximum - IncrementalValue.Initial) / (IncrementalValue.Initial * IncrementalValue.IncrementValue)

        # Something weird is that with the data we have if we try to floor it sometimes does not work correctly, so we first truncate and remove decimals, and then floor
        # We only do that because we are sure that max level is never gonna be a floor bigger than float (otherwise it could crash)
        if:
            Result:float = MaybeLevelResult?
            IntLevel:int = Floor[TruncateNumber(Result, 1)]
        then:
            set MaybeMaxLevel = option{IntLevel}

    MaybeMaxLevel

(IncrementalValue:incremental_value).ClampLevel(Level:int)<transacts>: int =
    if (Level >= 0):
        if (MaxLevel:=IncrementalValue.GetMaybeMaxLevel[]?):
            return Min(Level, MaxLevel)
        return Level
    return 0

# Returns the value with the level clamped to the max
(IncrementalValue:incremental_value).GetValueInClampedLevel(Level:int)<transacts>: float =
    ClampedLevel:=Max(0, Level)
    ValueInLevel:=IncrementalValue.GetValueInLevel(ClampedLevel)

    if (MaxIncrements:=IncrementalValue.MaybeMaxIncrements?):
        if (MaxIncrements >= 0 and Level > MaxIncrements):
            return IncrementalValue.GetValueInLevel(MaxIncrements)
    else if (Maximum:=IncrementalValue.MaybeMaximum?):
        if (ValueInLevel > Maximum, MaxLevel:=IncrementalValue.GetMaybeMaxLevel[]?):
            return IncrementalValue.GetValueInLevel(MaxLevel)

    return ValueInLevel

# Clamps the value in level to maximum (useful when you do not mind if the level is clamped or not and just want the value)
(IncrementalValue:incremental_value).GetClampedValueInLevel(Level:int)<computes><reads>: float =
    if (Level >= 0):
        ValueInLevel:=IncrementalValue.GetValueInLevel(Level)

        if (MaxIncrements:=IncrementalValue.MaybeMaxIncrements?):
            if (MaxIncrements >= 0 and Level > MaxIncrements):
                return IncrementalValue.GetValueInLevel(MaxIncrements)
        else if (Maximum:=IncrementalValue.MaybeMaximum?):
            return Min(ValueInLevel, Maximum)
            
        return ValueInLevel
    return 0.0

# Code to be able to pass more arguments to event subscriptions. Works for events that send an agent and that don't send anything.
(Listenable : listenable(agent)).SubscribeAgent(OutputFunc : tuple(agent, t)->void, ExtraData : t where t:type) : cancelable =
    Wrapper := wrapper_agent(t){ExtraData := ExtraData, OutputFunc := OutputFunc}
    Listenable.Subscribe(Wrapper.InputFunc)
    
wrapper_agent(t : type) := class():
    ExtraData : t;
    OutputFunc : tuple(agent, t) -> void
    InputFunc(Agent : agent):void = OutputFunc(Agent, ExtraData)
    
(Listenable : listenable(tuple())).SubscribeEmpty(OutputFunc : t -> void, ExtraData : t where t:type) : cancelable =
    Wrapper := wrapper_empty(t) {ExtraData := ExtraData, OutputFunc := OutputFunc}
    Listenable.Subscribe(Wrapper.InputFunc)
    
wrapper_empty(t : type) := class():
    ExtraData : t;
    OutputFunc : t -> void
    InputFunc():void = OutputFunc(ExtraData)

(Listenable : listenable(widget_message)).SubscribeButton(OutputFunc : tuple(widget_message, t)->void, ExtraData : t where t:type) : cancelable =
    Wrapper := wrapper_widget_message(t){ExtraData := ExtraData, OutputFunc := OutputFunc}
    Listenable.Subscribe(Wrapper.InputFunc)
    
wrapper_widget_message(t : type) := class():
    ExtraData : t;
    OutputFunc : tuple(widget_message, t) -> void
    InputFunc(Message : widget_message):void = OutputFunc(Message, ExtraData)

# Translates from an ID index to the enum value of the given array
GetIndexFromID(ID:t, Array:[]t where t:subtype(comparable))<transacts>:int =
    var Res:int = -1
    for (Index -> ArrayID : Array, ID = ArrayID):
        set Res = Index
    Res

# Translates from an ID value to the array index
GetIDFromIndex(Index:int, Array:[]t where t:subtype(comparable))<decides><transacts>:t =
    return Array[Index]

#Function to merge maps, preferred over ConcatenateMaps because the behavior is known and controlled   
MergeMaps(Lhs:[t]u, Rhs:[t]u where t:subtype(comparable),u:type)<transacts>:[t]u =
    var RetVal:[t]u = Lhs
    for (Id -> Key : Rhs, set RetVal[Id] = Key):
    RetVal

RemoveMapKey(Map:[t]u, ElementToRemove:t where t:subtype(comparable), u:type ):[t]u = 
    var NewMap:[t]u = map{}
    for ( Key -> Value:Map, Key <> ElementToRemove):
        set NewMap = ConcatenateMaps(NewMap, map{Key => Value})
    NewMap

operator'-'<public>((Map: [k]v where k:subtype(comparable), v: type), Target: k)<transacts>:[k]v=
    var Result: [k]v = map{}

    for(Key -> Value : Map, Key <> Target):
        if:
            set Result[Key] = Value

    return Result
   