
using { /Fortnite.com/Devices }
using { /Verse.org/Simulation }
using { /Verse.org/Simulation/Tags }
using { /UnrealEngine.com/Temporary/Diagnostics }
using { /UnrealEngine.com/Temporary/UI }

clash_store_device := class(store_device):

    CurrencyManagerTag:tag=currency_manager_tag{}
    var CurrencyManager:clash_currency_manager=clash_currency_manager{}
#=============================================================================================================================  
    OnBegin<override>()<suspends>:void=

        set StoreClass = ui_store_coins{}

        (super:)OnBegin()
        Sleep(1.0)
        CreativeObjects:=FindCreativeObjectsWithTag(CurrencyManagerTag)
        for(CreativeObject:CreativeObjects):
            if(Manager:=clash_currency_manager[CreativeObject]):
                set CurrencyManager = Manager

        loop:
            EventData:=CurrencyManager.CurrencyChangedEvent.Await()
            Player:=EventData.Player
            Currencies:=array{EventData.CurrentValue}
            CurrencyEventData:=ui_store_currency_change:
                Player:=Player
                Currencies:=Currencies
            PlayerCurrencyChangeEvent.Signal(CurrencyEventData)
#=============================================================================================================================  
    InitStore<override>(Player:player):void=
        StoreClass.OpenStore(Player, BaseStoreSettings, WidgetCloseEvent, PlayerCurrencyChangeEvent, WidgetButtonClickedEvent)
        #Update Player Currency
        Currency:=CurrencyManager.GetCurrency(Player)
        Currencies:=array{Currency}
        CurrencyEventData:=ui_store_currency_change:
            Player:=Player
            Currencies:=Currencies
        PlayerCurrencyChangeEvent.Signal(CurrencyEventData)
#=============================================================================================================================       
    HandleLoudButtonClick<override>(WidgetMessage:widget_message, OnClickPayload:ui_store_on_click_payload):void=
        (super:)HandleLoudButtonClick(WidgetMessage, OnClickPayload)
        MultiIndex:=OnClickPayload.MultiIndex
        Player:=WidgetMessage.Player

        if(Level_0_Index:=MultiIndex[0]):
            case (Level_0_Index):
                0 =>
                    Print("First case body")
                _ => # default
                    Print("Default case body")

            CurrencyManager.AddCurrency(-2.0, Player, currency_id.None)     
#=============================================================================================================================
    HandleQuietButtonClick<override>(WidgetMessage:widget_message, OnClickPayload:ui_store_on_click_payload):void=
        (super:)HandleQuietButtonClick(WidgetMessage, OnClickPayload)
        MultiIndex:=OnClickPayload.MultiIndex
        Player:=WidgetMessage.Player

        if(Level_0_Index:=MultiIndex[0]):
            case (Level_0_Index):
                0 =>
                    Print("First case body")
                _ => # default
                    Print("Default case body")
#=============================================================================================================================