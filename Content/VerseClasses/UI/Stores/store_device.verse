
using { /Fortnite.com/Devices }
using { /Fortnite.com/Characters }
using { /Fortnite.com/UI }
using { /Verse.org/Simulation }
using { /UnrealEngine.com/Temporary/Diagnostics }
using { /UnrealEngine.com/Temporary/UI }

store_device := class(creative_device):

    var StoreClass:ui_store=ui_store{}

    @editable
    BaseStoreSettings:ui_store_base_settings=ui_store_base_settings{}

    @editable
    OpenStoreTrigger:trigger_device=trigger_device{}

    @editable
    StoreCamera:gameplay_camera_fixed_point_device=gameplay_camera_fixed_point_device{}

    @editable
    bSetPlayerInvulnerable:logic=false

    #Players with store opened
    var PlayersCount:int=0

    #Events

    WidgetOpenevent:event(player)=event(player){}
    WidgetCloseEvent:event(player)=event(player){}
    #int: Button Index
    WidgetButtonClickedEvent:event(tuple(widget_message, ui_store_on_click_payload))=event(tuple(widget_message, ui_store_on_click_payload)){}

    PlayerCurrencyChangeEvent:event(ui_store_currency_change)=event(ui_store_currency_change){}
    PlayerEliminatedEvent:event(player)=event(player){}

    OnBegin<override>()<suspends>:void=
        OpenStoreTrigger.TriggeredEvent.Subscribe(OnAgentMaybeTryToOpenStore)

    OnAgentMaybeTryToOpenStore(Agent:?agent):void=
        if:
            ValidAgent:=Agent?
        then:
            Print("Openning Store")
            OpenStore(ValidAgent)

    OpenStore(Agent:agent):void=
        if(Player:=player[Agent]):
            InitStore(Player)
        if(PlayersCount<=0):
            spawn{EventHandler()}
        set PlayersCount += 1
        spawn{OnPlayerEliminatedEvent(Agent)}
        
    CloseStore(Player:player):void={}

    #Override this one in your implementation
    InitStore(Player:player):void={}

#=============================================================================================================================       
    HandleLoudButtonClick(WidgetMessage:widget_message, OnClickPayload:ui_store_on_click_payload):void=
        MultiIndex:=OnClickPayload.MultiIndex
        var SMultiIndex:string=""
        for(Level:MultiIndex):
            set SMultiIndex+="{Level}|"

        Print("Is Loud Button, Index:{SMultiIndex}")

        # if(Level_0_Index:=MultiIndex[0]):
        #     case (Level_0_Index):
        #         0 =>
        #             Print("First case body")
        #         _ => # default
        #             Print("Default case body")
#=============================================================================================================================
    HandleQuietButtonClick(WidgetMessage:widget_message, OnClickPayload:ui_store_on_click_payload):void=
        MultiIndex:=OnClickPayload.MultiIndex
        var SMultiIndex:string=""
        for(Level:MultiIndex):
            set SMultiIndex+="{Level}|"

        Print("Is Quiet Button, Index:{SMultiIndex}")

        # if(Level_0_Index:=MultiIndex[0]):
        #     case (Level_0_Index):
        #         0 =>
        #             Print("First case body")
        #         _ => # default
        #             Print("Default case body")
#=============================================================================================================================
    OnPlayerEliminatedEvent(Agent:agent)<suspends>:void=
        if:
            FortCharacter:=fort_character[Agent.GetFortCharacter[]]
            Player:=player[Agent]
        then:
            FortCharacter.EliminatedEvent().Await()
            PlayerEliminatedEvent.Signal(Player)

    EventHandler()<suspends>:void=
        TerminateHandler:=event(){}

        race:
            TerminateHandler.Await()

            loop:
                EventPlayer:=WidgetCloseEvent.Await()
                set PlayersCount -= 1
                if(PlayersCount<=0):
                    TerminateHandler.Signal()

            loop:
                EventPlayer:=PlayerEliminatedEvent.Await()
                WidgetCloseEvent.Signal(EventPlayer)
            loop:
                EventData:=WidgetButtonClickedEvent.Await()
                EventMessage:=EventData(0)
                EventClickPayload:=EventData(1)
                # EventPlayer:=EventMessage.Player
                Button:=EventMessage.Source
        
                if(LoudButton:=button_loud[Button]):
                    Print("Is Loud Button")
                    HandleLoudButtonClick(EventMessage, EventClickPayload)
                else:
                    if(QuietButton:=button_quiet[Button]):
                        Print("Is Quiet Button")
                        HandleQuietButtonClick(EventMessage, EventClickPayload)

#Add this loop on the Child class to bind with the currency manager
            # loop:
            #     EventData:=Currency Manager Await() Here


