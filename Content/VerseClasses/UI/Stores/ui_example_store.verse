using { /UnrealEngine.com/Temporary/UI }
using { /UnrealEngine.com/Temporary/SpatialMath }
using { /Fortnite.com/UI }
using { /Fortnite.com/Devices }
using { /Verse.org/Simulation }
using { /Verse.org/Colors }

ui_example_store:=class(ui_store):

#============================================================================================================================= 
    ConstructCanvas<override>(Player:player, CloseWidgetEvent:event(player), BaseStoreData:ui_store_base_data): ui_store_data = 
        #Create this Store data
        StoreData:=ui_store_data:
            BaseStoreData:=BaseStoreData

        #Create a button set (Loud & Quiet) inside an Overlay
        ButtonsData:=GenerateButtonSet("UPGRADE", "MAXED OUT")

        #Add this store data to the Base data
        BaseStoreData.MainStackBox.AddWidget(stack_box_slot{
            Widget:=ButtonsData(0)
            }
                )

        #Get Buttons and bind events
        ButtonL:=ButtonsData(1)
        ButtonQ:=ButtonsData(2)

        StoreData.AddButtonSet(ButtonL, ButtonQ)

        #Generate OnClick Data
        OnClickPayload:=ui_store_on_click_payload:
            StoreData:=StoreData
            MultiIndex:=array{0}

        #Bind Event Handler to buttons
        BindButtonsToHandler(ButtonL, ButtonQ, OnClickPayload)

        return StoreData
