using { /UnrealEngine.com/Temporary/UI }
using { /UnrealEngine.com/Temporary/SpatialMath }
using { /Fortnite.com/UI }
using { /Fortnite.com/Devices }
using { /Verse.org/Simulation }
using { /Verse.org/Colors }

ui_store:=class():

    var WidgetButtonClickHandler:event(tuple(widget_message, ui_store_on_click_payload))=event(tuple(widget_message, ui_store_on_click_payload)){}

    OpenStore(Player:player, BaseStoreSettings:ui_store_base_settings, CloseWidgetEvent:event(player), CurrencyChangeEvent:event(ui_store_currency_change), WidgetButtonClickedEvent:event(tuple(widget_message, ui_store_on_click_payload))):void=
        BaseStoreData:=ConstructBaseCanvas(Player, BaseStoreSettings, CloseWidgetEvent, CurrencyChangeEvent)
        set WidgetButtonClickHandler = WidgetButtonClickedEvent 
        StoreData:=ConstructCanvas(Player, CloseWidgetEvent, BaseStoreData)
        if(PlayerUI:=GetPlayerUI[Player]):
            Slot:=player_ui_slot{InputMode:=ui_input_mode.All}
            PlayerUI.AddWidget(BaseStoreData.Canvas, Slot)
#============================================================================================================================= 
    ConstructCanvas<protected>(Player:player, CloseWidgetEvent:event(player), BaseStoreData:ui_store_base_data): ui_store_data = 
        #Create this Store data
        StoreData:=ui_store_data:
            BaseStoreData:=BaseStoreData

        return StoreData
#=============================================================================================================================       
    ConstructBaseCanvas<protected>(Player:player, BaseStoreSettings:ui_store_base_settings, WidgetRemovedEvent:event(player), CurrencyChangeEvent:event(ui_store_currency_change)): ui_store_base_data =

        CloseButton:=button_loud{DefaultText:=StringToMessage(BaseStoreSettings.CloseButtonString)}

        NoMoneyTextBlock:=text_block:
            DefaultText:=StringToMessage(BaseStoreSettings.NoMoneyString)
            DefaultTextColor:=NamedColors.Red
        
        NoMoneyTextBlock.SetVisibility(widget_visibility.Collapsed)

        CurrencyTextBlock:=text_block:
            DefaultText:=StringToMessage("0")
            DefaultTextColor:=BaseStoreSettings.CurrencyTextColor

        MainStackBox:=stack_box:
            Orientation:=orientation.Vertical
            Slots:=array:
                # Header
                stack_box_slot:
                    HorizontalAlignment:=horizontal_alignment.Fill
                    VerticalAlignment:=vertical_alignment.Fill
                    Widget:=overlay:
                        Slots:=array:
                            # Title
                            overlay_slot:
                                VerticalAlignment:=vertical_alignment.Top
                                HorizontalAlignment:=horizontal_alignment.Center
                                Padding:=margin{Top:=20.0, Left:=20.0, Right:=20.0, Bottom:=20.0}
                                Widget:=text_block{DefaultText:=StringToMessage(BaseStoreSettings.TitleString), DefaultTextColor:=BaseStoreSettings.TitleTextColor,
                                    DefaultJustification:=text_justification.Center}
                            # Back Button
                            overlay_slot:
                                VerticalAlignment:=vertical_alignment.Center
                                HorizontalAlignment:=horizontal_alignment.Right
                                Padding:=margin{Right:=BaseStoreSettings.BackButtonRightPadding}
                                Widget:=CloseButton
                            # Currency Slot
                            overlay_slot:
                                # VerticalAlignment:=vertical_alignment.Center
                                # HorizontalAlignment:=horizontal_alignment.Left
                                # Padding:=margin{Left:=150.0}
                                # Widget:=overlay:
                                #     Slots:=array:
                                #         # Background
                                #         <#overlay_slot:
                                #             HorizontalAlignment:=horizontal_alignment.Fill
                                #             VerticalAlignment:=vertical_alignment.Fill
                                #             Widget:=texture_block{DefaultImage:=FindCurrencySlotBackground(DisplayCurrency), DefaultDesiredSize:=vector2{X:=300.0, Y:=65.0}}
                                #         # Icon
                                #         overlay_slot:
                                #             HorizontalAlignment:=horizontal_alignment.Left
                                #             VerticalAlignment:=vertical_alignment.Center
                                #             Padding:=margin{Left:=15.0}
                                #             Widget:=texture_block{DefaultImage:=FindCurrencyIcon(DisplayCurrency), DefaultDesiredSize:= vector2{X:=45.0, Y:=45.0}}#>
                                #         # Text
                                #         overlay_slot:
                                #             HorizontalAlignment:=horizontal_alignment.Fill
                                #             VerticalAlignment:=vertical_alignment.Center
                                #             Padding:=margin{Left:=90.0, Right:=10.0}
                                #             Widget:=CurrencyTextBlock

                                VerticalAlignment:=vertical_alignment.Center
                                HorizontalAlignment:=horizontal_alignment.Fill
                                Padding:=margin{Left:=0.0, Top:=0.0, Right:=0.0, Bottom:=0.0}
                                Widget:=stack_box:
                                    Orientation:=orientation.Horizontal
                                    Slots:=array:
                                        
                                        #Coin$ counter
                                        stack_box_slot:
                                            VerticalAlignment:=vertical_alignment.Center
                                            HorizontalAlignment:=horizontal_alignment.Left
                                            Distribution:=option{0.33}
                                            Padding:=margin{Left:=30.0, Top:=0.0, Right:=0.0, Bottom:=0.0}
                                            Widget:=overlay:
                                                Slots:=array:
            
                                                    #Makeshift size box
                                                    #overlay_slot:
                                                        #VerticalAlignment:=vertical_alignment.Center
                                                        #HorizontalAlignment:=horizontal_alignment.Center
                                                        ##Distribution:=option{0.33}
                                                        #Padding:=margin{Left:=0.0, Top:=0.0, Right:=0.0, Bottom:=0.0}
                                                        #Widget:=texture_block{DefaultImage:=Textures.General.T_GenericSquare, DefaultDesiredSize:=vector2{X:=650.0, Y:=50.0}, DefaultTint:=MakeColorFromHex("#FFFFFFFF")}
            
                                                    #Coin$ counter
                                                    overlay_slot:
                                                        VerticalAlignment:=vertical_alignment.Center
                                                        HorizontalAlignment:=horizontal_alignment.Left
                                                        #Distribution:=option{0.33}
                                                        Padding:=margin{Left:=0.0, Top:=0.0, Right:=0.0, Bottom:=0.0}
                                                        Widget:=texture_block{DefaultImage:=Textures.Currency.T_GoldSlot, DefaultDesiredSize:=vector2{X:=180.0, Y:=50.0}, DefaultTint:=MakeColorFromHex("#FFFFFFFF")}
            
                                                    #Coin$ counter layout
                                                    overlay_slot:
                                                        VerticalAlignment:=vertical_alignment.Center
                                                        HorizontalAlignment:=horizontal_alignment.Left
                                                        #Distribution:=option{1.0}
                                                        Padding:=margin{Left:=0.0, Top:=0.0, Right:=0.0, Bottom:=0.0}
                                                        Widget:=stack_box:
                                                            Orientation:=orientation.Horizontal
                                                            Slots:=array:
            
                                                                #Coin$ icon
                                                                stack_box_slot:
                                                                    VerticalAlignment:=vertical_alignment.Center
                                                                    HorizontalAlignment:=horizontal_alignment.Left
                                                                    Distribution:=option{1.0}
                                                                    Padding:=margin{Left:=5.0, Top:=0.0, Right:=0.0, Bottom:=0.0}
                                                                    Widget:=texture_block{DefaultImage:=Textures.Currency.T_Gold, DefaultDesiredSize:=vector2{X:=40.0, Y:=40.0}, DefaultTint:=MakeColorFromHex("#FFFFFFFF")}
            
                                                                #Coin$ value
                                                                stack_box_slot:
                                                                    VerticalAlignment:=vertical_alignment.Center
                                                                    HorizontalAlignment:=horizontal_alignment.Left
                                                                    Distribution:=option{1.0}
                                                                    Padding:=margin{Left:=10.0, Top:=0.0, Right:=0.0, Bottom:=0.0}
                                                                    Widget:=CurrencyTextBlock

        Canvas:=canvas:
            Slots:=array:
                canvas_slot:
                    Anchors:=anchors{Minimum:=vector2{X:=0.0,Y:=0.0}, Maximum:=vector2{X:=1.0,Y:=1.0}}
                    Alignment:=vector2{X:=0.0, Y:=0.0}
                    Offsets:=margin{Top:=30.0, Left:=30.0, Right:=30.0, Bottom:=30.0}
                    Widget:=overlay:
                        Slots:=array:
                            # Full Background
                            overlay_slot:
                                VerticalAlignment:=vertical_alignment.Fill
                                HorizontalAlignment:=horizontal_alignment.Fill
                                Widget:=texture_block{DefaultImage:=Textures.General.T_Background}
                            # Main
                            overlay_slot:
                                VerticalAlignment:=vertical_alignment.Fill
                                HorizontalAlignment:=horizontal_alignment.Fill
                                Padding:=margin{Top:=30.0, Left:=30.0, Right:=30.0, Bottom:=30.0}
                                Widget:=MainStackBox
                            #No money message
                            overlay_slot:
                                VerticalAlignment:=vertical_alignment.Bottom
                                HorizontalAlignment:=horizontal_alignment.Center
                                Padding:=margin{Top:=50.0, Left:=30.0, Right:=30.0, Bottom:=0.0}
                                Widget:=NoMoneyTextBlock

        BaseStoreData:=ui_store_base_data:
            Canvas:=Canvas
            CloseButton:=CloseButton
            MainStackBox:=MainStackBox
            CurrencyTextBlock:=CurrencyTextBlock
            NoMoneyTextBlock:=NoMoneyTextBlock

        #Events
        spawn{WidgetEventHandler(Player, BaseStoreData, WidgetRemovedEvent, CurrencyChangeEvent)}

        return BaseStoreData
#=============================================================================================================================
    GenerateButtonSet(LoudButtonText:string, QuietButtonText:string):tuple(overlay, button_loud, button_quiet)=
        ButtonL:=button_loud{DefaultText:=StringToMessage(LoudButtonText)}
        ButtonQ:=button_quiet{DefaultText:=StringToMessage(QuietButtonText)}
        ButtonQ.SetVisibility(widget_visibility.Collapsed)

        Overlay:=overlay{
            Slots:=array{
                overlay_slot:
                    Widget:=ButtonL
                    VerticalAlignment:=vertical_alignment.Center
                    HorizontalAlignment:=horizontal_alignment.Center
                overlay_slot:
                    Widget:=ButtonQ
                    VerticalAlignment:=vertical_alignment.Center
                    HorizontalAlignment:=horizontal_alignment.Center
            }
        }

        return (Overlay, ButtonL, ButtonQ)
#=============================================================================================================================
    BindButtonsToHandler(Loud:button_loud, Quiet:button_quiet, OnClickPayload:ui_store_on_click_payload):void=
        Loud.OnClick().SubscribeButton(OnButtonStoreClickedEvent, OnClickPayload)
        Quiet.OnClick().SubscribeButton(OnButtonStoreClickedEvent, OnClickPayload)
#=============================================================================================================================
    CheckIfButtonShouldBeEnabled(StoreData:ui_store_data, ButtonIndex:int):void=
        var Check:logic = true
        #E.g: button should be quiet if player currency is not enought
        <#
            if(Currency < RequiredCurrency):
                set Check = false
        #>
        StoreData.ToggleButtonVisibility(Check, ButtonIndex)
#=============================================================================================================================
    #Note: override this method to mix all currencies
    UpdateWidgetCurrencyData(BaseStoreData:ui_store_base_data, CurrencyData:ui_store_currency_change):void=
        EventCurrencies:=CurrencyData.Currencies
        if:
            Gold:=EventCurrencies[0]
        then:
            GoldMessage:=StringToMessage(ToNotationString(Gold))
            BaseStoreData.CurrencyTextBlock.SetText(GoldMessage)
#=============================================================================================================================
    #Events
    OnButtonStoreClickedEvent(WidgetMessage:widget_message, OnClickPayload:ui_store_on_click_payload):void=
        Button:=WidgetMessage.Source

        if(LoudButton:=button_loud[Button]):
            Print("Is Loud Button")
            HandleLoudButtonClick(OnClickPayload)
        else:
            if(QuietButton:=button_quiet[Button]):
                Print("Is Quiet Button")
                HandleQuietButtonClick(OnClickPayload)

        WidgetButtonClickHandler.Signal(WidgetMessage, OnClickPayload)
#=============================================================================================================================       
    HandleLoudButtonClick(OnClickPayload:ui_store_on_click_payload):void=
        MultiIndex:=OnClickPayload.MultiIndex
        var SMultiIndex:string=""
        for(Level:MultiIndex):
            set SMultiIndex+="{Level}|"

        Print("Is Loud Button, Index:{SMultiIndex}")
#=============================================================================================================================
    HandleQuietButtonClick(OnClickPayload:ui_store_on_click_payload):void=
        MultiIndex:=OnClickPayload.MultiIndex
        var SMultiIndex:string=""
        for(Level:MultiIndex):
            set SMultiIndex+="{Level}|"

        Print("Is Quiet Button, Index:{SMultiIndex}")
#=============================================================================================================================
    RemoveWidget(Player:player, Canvas:canvas):void=
        if(PlayerUI:=GetPlayerUI[Player]):
            PlayerUI.RemoveWidget(Canvas)
#=============================================================================================================================
    WidgetEventHandler(Player:player, BaseStoreData:ui_store_base_data, CloseWidgetEvent:event(player), CurrencyChangeEvent:event(ui_store_currency_change))<suspends>:void=
        TerminateHandler:=event(){}

        defer:
            block:
                RemoveWidget(Player, BaseStoreData.Canvas)
                Print("Close Widget Corutine Terminated")
        race:
            TerminateHandler.Await()

            loop:
                EventPlayer:=CloseWidgetEvent.Await()
                Print("Widget Close Event Recieved")
                if(EventPlayer = Player):
                    TerminateHandler.Signal()

            loop:
                BaseStoreData.CloseButton.OnClick().Await()
                Print("Close Button Clicked")
                CloseWidgetEvent.Signal(Player)
            loop:
                EventData:=CurrencyChangeEvent.Await()
                EventPlayer:=EventData.Player
                if(EventPlayer=Player):
                    UpdateWidgetCurrencyData(BaseStoreData, EventData)


                    


