using { /UnrealEngine.com/Temporary/UI }
using { /UnrealEngine.com/Temporary/SpatialMath }
using { /Fortnite.com/UI }
using { /Fortnite.com/Devices }
using { /Verse.org/Simulation }
using { /Verse.org/Colors }

ui_store_coins_on_click_payload:=struct():
    CoinsStoreData:ui_store_coins_data
    Index:int

ui_store_coins_data:=class(ui_store_data):
    var Tabs:[]stack_box

ui_store_coins_tab:=struct:
    TabName:string
    SubTabs:[]string
    ItemName:string

ui_store_coins:=class(ui_store):

    SwitchTabEvent:event(player)=event(player){}

    MainTabs:[]ui_store_coins_tab=array{
        ui_store_coins_tab:
            TabName:="Enemies"
            SubTabs:=array{"By Assist","By Elimination"}
            ItemName:="Enemy Lvl"

        ui_store_coins_tab:
            TabName:="Others"
            SubTabs:=array{"By Assist","By Elimination"}
            ItemName:="Other Lvl"
        }

    ConstructCanvas<override>(Player:player, CloseWidgetEvent:event(player), BaseStoreData:ui_store_base_data): ui_store_data = 

        #Create this Store data
        StoreData:=ui_store_coins_data:
            BaseStoreData:=BaseStoreData
            Tabs:=array{}

        #Widget Main Container
        CoinsStoreWidget:=stack_box:
            Orientation:=orientation.Vertical

        #Tab Buttons
        TabButtonsData:=CreateButtonsTab(Player, MainTabs)

        CoinsStoreWidget.AddWidget(
            stack_box_slot:
                Widget:=TabButtonsData(0)
                # Padding:=margin{Left:=200.0, Top:=5.0, Right:=100.0, Bottom:=5.0}
        )

        var TabIndex:int=0

        for(TargetTab:MainTabs):
            ItemContainer:=CreateItemContainer(Player, TabIndex, TargetTab, StoreData)
            StackBox:=ItemContainer
            set StoreData.Tabs+=array{StackBox}
            StackBox.SetVisibility(widget_visibility.Collapsed)
            CoinsStoreWidget.AddWidget(
                stack_box_slot:
                    Widget:=StackBox
            )
            set TabIndex+=1
                
        if(FirstTab:=StoreData.Tabs[0]):
            FirstTab.SetVisibility(widget_visibility.Visible)

        #Add this store data to the Base data
        BaseStoreData.MainStackBox.AddWidget(stack_box_slot{
            Widget:=CoinsStoreWidget
            HorizontalAlignment:=horizontal_alignment.Fill
            VerticalAlignment:=vertical_alignment.Fill
            Distribution:=option{1.0}
            }
                )

        #Bind Events

        #Tab Buttons, Switch Tab when button is pressed
        var Index:int=0
        for(Button:TabButtonsData(1)):
            CoinsOnClickPayload:=ui_store_coins_on_click_payload:
                CoinsStoreData:=StoreData
                Index:=Index
            Button.OnClick().SubscribeButton(HandleCoinStoreEvents, CoinsOnClickPayload)
            set Index+=1

        return StoreData


    CreateButtonsTab(Player:player, Tabs:[]ui_store_coins_tab):tuple(stack_box, []button_loud)=
        ButtonStackBox:=stack_box:
            Orientation:= orientation.Horizontal

        var Buttons_Loud:[]button_loud=array{}

        for(Tab:Tabs):
            TabButton:=button_loud{DefaultText:=StringToMessage(Tab.TabName)}
            Slot:=stack_box_slot:
                Widget:=TabButton
                Padding:=margin{Left:=10.0, Right:=10.0}   
                HorizontalAlignment:=horizontal_alignment.Fill
                VerticalAlignment:=vertical_alignment.Fill    
                Distribution:=option{1.0}     
            ButtonStackBox.AddWidget(Slot)
            set Buttons_Loud+=array{TabButton}
    
        return (ButtonStackBox, Buttons_Loud) 

    CreateItemContainer(Player:player, TabIndex:int, Tab:ui_store_coins_tab, CoinsStoreData:ui_store_coins_data):stack_box=

        ItemContainer:=stack_box:
            Orientation:= orientation.Vertical

        SubTabsStackBox:=stack_box:
            Orientation:= orientation.Horizontal

        for(SubTab:Tab.SubTabs):
            SubTabTextBlock:=text_block{DefaultText:=StringToMessage(SubTab)}
            Slot:=stack_box_slot:
                Widget:=SubTabTextBlock
                Padding:=margin{Right:=50.0}
                HorizontalAlignment:=horizontal_alignment.Fill
                VerticalAlignment:=vertical_alignment.Fill
            SubTabsStackBox.AddWidget(Slot)

        ItemContainer.AddWidget(
            NewSlot:=stack_box_slot:
                Widget:=SubTabsStackBox
                HorizontalAlignment:=horizontal_alignment.Fill
                VerticalAlignment:=vertical_alignment.Fill
        )

        for(Level:int=0..3):
            ButtonMultiIndex:=array{TabIndex, Level}
            Item:=CreateItem(Player, ButtonMultiIndex, Tab, CoinsStoreData)
            ItemContainer.AddWidget(
                NewItemSlot:=stack_box_slot:
                    Widget:=Item
                    HorizontalAlignment:=horizontal_alignment.Fill
                    VerticalAlignment:=vertical_alignment.Fill
            )

        return ItemContainer


    CreateItem(Player:player, MultiIndex:[]int, Tab:ui_store_coins_tab, CoinsStoreData:ui_store_coins_data):stack_box=

        var Button_0_Index:int=0
        var Button_1_Index:int=1

        Padding:=margin{Right:=10.0}
        
        ButtonsData_0:=GenerateButtonSet("Upgrade\n$100", "MAXED OUT")
        ButtonL0:=ButtonsData_0(1)
        ButtonQ0:=ButtonsData_0(2)

        #Generate OnClick Data
        OnClickPayload_0:=ui_store_on_click_payload:
            StoreData:=CoinsStoreData
            #Multidimensional Index = Tab->Item->Button Index
            MultiIndex:=MultiIndex+array{Button_0_Index}
        
        #Bind Event Handler to buttons
        BindButtonsToHandler(ButtonL0, ButtonQ0, OnClickPayload_0)

        ButtonsData_1:=GenerateButtonSet("Upgrade\n$100", "MAXED OUT")
        ButtonL1:=ButtonsData_1(1)
        ButtonQ1:=ButtonsData_1(2)

        #Generate OnClick Data
        OnClickPayload_1:=ui_store_on_click_payload:
            StoreData:=CoinsStoreData
            #Multidimensional Index = Tab->Item->Button Index
            MultiIndex:=MultiIndex+array{Button_1_Index}
                
        #Bind Event Handler to buttons
        BindButtonsToHandler(ButtonL1, ButtonQ1, OnClickPayload_1)

        ItemsStackBox:=stack_box:
            Orientation:= orientation.Horizontal
            Slots:=array:
                stack_box_slot:
                    Widget:=text_block{DefaultText:=StringToMessage("{Tab.ItemName} {Button_0_Index}")}
                    Padding:=Padding
                stack_box_slot:
                    Widget:=ButtonsData_0(0)#button_loud{DefaultText:=(StringToMessage("Upgrade\n$100"))}
                    Padding:=Padding
                    Distribution:=option{1.0}
                stack_box_slot:
                    Widget:=text_block{DefaultText:=StringToMessage("{Tab.ItemName} {Button_1_Index}")}
                    Padding:=Padding
                stack_box_slot:
                    Widget:=ButtonsData_1(0)#button_loud{DefaultText:=(StringToMessage("Upgrade\n$100"))}
                    Padding:=Padding
                    Distribution:=option{1.0}

        return ItemsStackBox

    HandleCoinStoreEvents(Message:widget_message, CoinsStoreData:ui_store_coins_on_click_payload):void=
        Index:=CoinsStoreData.Index
        Tabs:=CoinsStoreData.CoinsStoreData.Tabs     

        var CurrentIndex:int=0

        for(TargetTab:Tabs):
            if(CurrentIndex = Index):
                TargetTab.SetVisibility(widget_visibility.Visible)
            else:
                TargetTab.SetVisibility(widget_visibility.Collapsed)

            set CurrentIndex+=1




