using { /Verse.org/Simulation }
using { /Verse.org/Colors }
using { /Fortnite.com/Devices }
using { /Fortnite.com/UI }
using { /UnrealEngine.com/Temporary/UI }

ui_store_base_settings:=class<concrete>():
    @editable
    CurrencyTextColor:color = NamedColors.Black

    @editable
    TitleTextColor:color = NamedColors.White

    @editable
    CloseButtonString:string = "X"

    @editable
    TitleString:string ="Store Name"

    @editable
    BackButtonRightPadding:float = 350.0

    @editable
    SuccessSound : audio_player_device = audio_player_device{}

    @editable
    FailSound : audio_player_device = audio_player_device{}

    @editable
    NoMoneyString:string = "You don't have enough currency!"

ui_store_base_data := struct():
    Canvas:canvas
    MainStackBox:stack_box
    CloseButton:button_loud
    CurrencyTextBlock:text_block
    NoMoneyTextBlock:text_block

ui_store_data:=class():
    BaseStoreData:ui_store_base_data
    var Button_Loud:[]button_loud = array{}
    var Button_Quiet:[]button_quiet = array{}

ui_store_on_click_payload:=struct():
    StoreData:ui_store_data
    MultiIndex:[]int #Multi dimensional index

(UIStoreData:ui_store_data).AddButtonSet(Loud:button_loud, Quiet:button_quiet):void=
    set UIStoreData.Button_Loud += array{Loud}
    set UIStoreData.Button_Quiet += array{Quiet}

    Print("Buttons Count: {UIStoreData.Button_Loud.Length}")

(UIStoreData:ui_store_data).ToggleButtonVisibility(On:logic, ButtonIndex:int):void=
    if:
        Loud:=UIStoreData.Button_Loud[ButtonIndex]
        Quiet:=UIStoreData.Button_Quiet[ButtonIndex]
    then:
        var NewVisibilityL:widget_visibility=widget_visibility.Visible
        var NewVisibilityQ:widget_visibility=widget_visibility.Collapsed

        if(not On?):
            set NewVisibilityL = widget_visibility.Collapsed
            set NewVisibilityQ = widget_visibility.Visible

        Loud.SetVisibility(NewVisibilityL)
        Quiet.SetVisibility(NewVisibilityQ)

ui_store_currency_change:=struct():
    Player:player
    Currencies:[]float